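; BP_Environment_Light_Rig is a UEFN environment light device that manages exposure post-process and publishes sunlight direction to materials.
; UpdateComponents enables either LumenExposure or BasicExposure from GI quality, platform and cloud checks, defaulting to Lumen when UseExposureLogic is off.
; It runs at construction, BeginPlay and GI quality changes.
; SetMPC writes the negated directional light forward vector to Fortnite's shared material parameter collection.
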(event EventBeginPlay
  (Default|UpdateComponents)
  (CallFunction|SetMPC))

(event Custom|Update%20Directional%20Light%20MPC
  (CallFunction|SetMPC))

(event Scalability|EventOnGlobalIlluminationQualityChanged
  (Default|UpdateComponents))

(event EditorPreview|EventSetupForEditorPreview
  ;[1 1 1] Editor Setup: Hide billboard, we've already have the editor only mesh for the usage
  (for _array_element (Actor|GetComponentsByClass self "/Script/Engine.BillboardComponent")
    (Rendering|SetVisibility _array_element false false)))

(fn IsLumenSupported
  (bind _returnvalue (TouchPlatform|IsNintendoSwitch))
  (bind _returnvalue_2 (FortGlobals|IsRunninginCloud))
  (bind _returnvalue_3 (TouchPlatform|IsMobileGame))
  (bind _returnvalue_4 (not _returnvalue_3))
  (bind _returnvalue_5 (not _returnvalue))
  (bind _returnvalue_6 (or _returnvalue_4 _returnvalue_2))
  (bind _returnvalue_7 (Development|GetConsoleVariableIntValue "sg.GlobalIlluminationQuality"))
  (Fortnite|FortShippingLog self (Utilities|String|Append "UEFN Environment Light Device: Is Lumen Supported = " (Utilities|String|ToString(Boolean) (Math|Boolean|ANDBoolean (>= _returnvalue_7 2) _returnvalue_5 _returnvalue_6))))
  (return (Math|Boolean|ANDBoolean (>= _returnvalue_7 2) _returnvalue_5 _returnvalue_6)))

(fn SetMPC
  (Rendering|Material|SetVectorParameterValue "/Game/Packages/Fortress_SharedMaterials/GlobalMaterialParameters/FortniteMaterialParameters.FortniteMaterialParameters" "SunlightDirectionalVector" (Math|Conversions|ToLinearColor(Vector) (* (Transformation|GetForwardVector (Variables|Default|GetDirectionalLight)) -1.0)) :!class /Script/Engine.KismetMaterialLibrary))

(fn UpdateComponents
  (bind _lumenexposure (Variables|Default|GetLumenExposure))
  (bind _basicexposure (Variables|Default|GetBasicExposure))
  (Fortnite|FortShippingLog self "UEFN Environment Light Device: Update Components")
  (bind _islumensupported (CallFunction|IsLumenSupported))
  (Utilities|FlowControl|Sequence
    (:then_0
      (bind _output_get (Variables|/Script/Engine.PostProcessComponent|SetbEnabled (or _islumensupported (not (Variables|Default|GetUseExposureLogic))) _lumenexposure))
      (Rendering|SetVisibility _lumenexposure _output_get false))
    (:then_1
      (bind _output_get_2 (Variables|/Script/Engine.PostProcessComponent|SetbEnabled (and (not _islumensupported) (Variables|Default|GetUseExposureLogic)) _basicexposure))
      (Rendering|SetVisibility _basicexposure _output_get_2 false))))

(fn ConstructionScript
  (|CallParentFunction)
  (Default|UpdateComponents)
  (CallFunction|SetMPC))
