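; Player controller component that keeps Creative map marker visibility matched to the owning player's use restrictions.
; Registered markers are stored in a set, updated immediately, and re-evaluated when their visible-on-use-restriction state changes.
; On BeginPlay it retries on a short timer until a minigame is found, then binds class, team and mutator-list changes.
; Class or team changes for the owning controller, and any mutator-list update, refresh every stored marker.
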(event EventTick
  (params (DeltaSeconds :type Float%20%28single-precision%29)))

(event Custom|RegisterMarkerComponent
  (params (MarkerComponent :type Creative%20Map%20Marker%20Component%20Object%20Reference))
  (Utilities|Set|Add (Variables|Default|GetMarkerComponents) MarkerComponent)
  (CallFunction|UpdateVisibility MarkerComponent)
  (Default|BindEventtoIsVisibleonUseRestrictionRep MarkerComponent (EventDispatchers|CreateEvent :!event Custom|OnIsVisibleOnUseRestrictionChanged)))

(event Custom|OnMinigameBound
  (CallFunction|BindPlayerStateMutatorListUpdated)
  (CallFunction|ReplicatedRefreshAll))

(event EventBeginPlay
  (|CallParentFunction)
  (CallFunction|TryBindMinigame))

(event Custom|ReplicatedUpdateVisibility
  (params (MarkerComponent :type Creative%20Map%20Marker%20Component%20Object%20Reference))
  (CallFunction|UpdateVisibility MarkerComponent))

(event Custom|UpdateVisibility
  (params (MarkerComponent :type Creative%20Map%20Marker%20Component%20Object%20Reference))
  (Utilities|IsValid MarkerComponent
    (:Is%20Valid
      (Class|CreativeMapMarkerComponent|SetVisibilityonUseRestriction MarkerComponent (Components|GetOwner)))))

(event Custom|OnClassChanged
  (params (PlayerState :type Fort%20Player%20State%20Object%20Reference) (ClassSlotIndex :type Byte))
  (if (CallFunction|HasMatchingPlayerController PlayerState)
    (CallFunction|ReplicatedRefreshAll)))

(event Custom|OnTeamChanged
  (params (PlayerState :type Fort%20Player%20State%20Object%20Reference) (TeamId :type Byte) (OldTeamId :type Byte) (bTeamChangedWithoutRespawn :type Boolean))
  (if (CallFunction|HasMatchingPlayerController PlayerState)
    (CallFunction|ReplicatedRefreshAll)))

(event Custom|ReplicatedRefreshAll
  (CallFunction|RefreshAll))

(event Custom|OnMutatorListUpdatedDelegate_Event
  (CallFunction|RefreshAll))

(event Custom|OnIsVisibleOnUseRestrictionChanged
  (params (MarkerComponent :type Creative%20Map%20Marker%20Component%20Object%20Reference))
  (CallFunction|UpdateVisibility MarkerComponent))

(fn BindPlayerStateMutatorListUpdated
  (bind _asplayer_controller (Utilities|Casting|CastToPlayerController (Components|GetOwner)))
  (bind _asfort_player_state (Utilities|Casting|CastToFortPlayerState (Class|AthenaWidgetLibrary|GetRelevantPlayerState _asplayer_controller)))
  (FortPlayerState|BindEventtoOnMutatorListUpdatedDelegate _asfort_player_state (EventDispatchers|CreateEvent :!event OnMutatorListUpdatedDelegate_Event)))

(fn HasMatchingPlayerController
  (params (Actor :type Actor%20Object%20Reference))
  (Utilities|IsValid Actor
    (:Is%20Valid
      (return (== (Fortnite|GetControllerfromActor Actor) (Components|GetOwner))))
    (:Is%20Not%20Valid
      (return false))))

(fn RefreshAll
  (bind _valuepin (Utilities|Set|ForEachLoop(Set) (Variables|Default|GetMarkerComponents)))
  (Utilities|IsValid _valuepin
    (:Is%20Valid
      (CallFunction|UpdateVisibility _valuepin))))

(fn TryBindMinigame
  (bind _returnvalue (Minigame|GetMinigameforActor (Components|GetOwner)))
  (Utilities|IsValid _returnvalue
    (:Is%20Valid
      (Minigame|BindEventtoOnPlayerClassApplied _returnvalue (EventDispatchers|CreateEvent :!event OnClassChanged))
      (Minigame|BindEventtoPlayerChangedTeamDelegate _returnvalue (EventDispatchers|CreateEvent :!event OnTeamChanged))
      (CallFunction|OnMinigameBound))
    (:Is%20Not%20Valid
      (Utilities|Time|SetTimerbyEvent (EventDispatchers|CreateEvent :!event TryBindMinigame) 0.25))))
