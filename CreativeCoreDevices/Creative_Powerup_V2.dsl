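; Base Creative powerup device that grants, or clears, a timed or infinite gameplay effect on players.
; Mesh or radius-sphere overlaps and events run server-side pickup, filtering by team/class and sending GAS events to the chosen recipients.
; SetupEffectData builds specs from overridable duration, set-by-caller magnitude and tags, including persist-on-elimination.
; A replicated two-flag collectible array drives visibility, while server timers handle respawn and client calls play pickup effects.
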(event Custom|Client_PlayAmbienLoopSound
  ;[1 0.8036 0 0.3] Effects
  (if (Variables|UserOptions|Basic|GetbAmbientAudio)
    (if (not (Audio|Components|Audio|IsPlaying (Variables|Components|GetAmbient%20Loop)))
      (Audio|Components|Audio|FadeIn (Variables|Components|GetAmbient%20Loop) 0.5 1.0 0.0 "Linear" :!class /Script/Engine.AudioComponent))
    (else
      (CallFunction|FadeOutAmbientSound))))

(event Custom|Play%20Ambient%20Loop%20Sound
  ;[1 0.8036 0 0.3] Effects
  (if (Variables|UserOptions|Basic|GetbAmbientAudio)
    (if (not (Audio|Components|Audio|IsPlaying (Variables|Components|GetAmbient%20Loop)))
      (Audio|Components|Audio|FadeIn (Variables|Components|GetAmbient%20Loop) 0.5 1.0 0.0 "Linear" :!class /Script/Engine.AudioComponent))
    (else
      (CallFunction|FadeOutAmbientSound))))

(event Custom|PlayCollectionFX
  ;[1 0.8036 0 0.3] Effects
  (Utilities|FlowControl|DoOnce false
    (:Completed
      (bind _returnvalue (Effects|SpawnEmitteratLocation :EmitterTemplate (Variables|Default|GetPickup%20VFX) :Location (+ (Transformation|GetWorldLocation (Variables|Components|GetStaticMeshComponent)) (Variables|Default|GetPickupVFXOffset)) :Scale "1.000000,1.000000,1.000000" :bAutoDestroy true :PoolingMethod "AutoRelease" :bAutoActivateSystem true))
      (Effects|Components|ParticleSystem|SetColorParameter _returnvalue "Color" (Variables|Default|GetVFX%20Particle%20Color)))))

(event Custom|ResetPlayCollectionFXGate
  ;[1 0.8036 0 0.3] Effects
  (Utilities|FlowControl|DoOnce false
    (:Completed
      (bind _returnvalue (Effects|SpawnEmitteratLocation :EmitterTemplate (Variables|Default|GetPickup%20VFX) :Location (+ (Transformation|GetWorldLocation (Variables|Components|GetStaticMeshComponent)) (Variables|Default|GetPickupVFXOffset)) :Scale "1.000000,1.000000,1.000000" :bAutoDestroy true :PoolingMethod "AutoRelease" :bAutoActivateSystem true))
      (Effects|Components|ParticleSystem|SetColorParameter _returnvalue "Color" (Variables|Default|GetVFX%20Particle%20Color)))))

(event Custom|Play%20Pick%20Up%20Sound
  ;[1 0.8036 0 0.3] Effects
  (if (Variables|UserOptions|GetPickUpAudio)
    (Audio|PlaySoundatLocation (Variables|Default|GetPickupSound) (Transformation|GetActorLocation))))

(event Custom|Fade%20Out%20Ambient%20Sound
  ;[1 0.8036 0 0.3] Effects
  (if (Audio|Components|Audio|IsPlaying (Variables|Components|GetAmbient%20Loop))
    (Audio|Components|Audio|FadeOut (Variables|Components|GetAmbient%20Loop) 0.2 0.0 "Linear" :!class /Script/Engine.AudioComponent)))

(event OnComponentBeginOverlap(StaticMeshComponent)
  (params (OverlappedComponent :type Primitive%20Component%20Object%20Reference) (OtherActor :type Actor%20Object%20Reference) (OtherComp :type Primitive%20Component%20Object%20Reference) (OtherBodyIndex :type Integer) (bFromSweep :type Boolean) (SweepResult :type Hit%20Result%20Structure%20%28by%20ref%29))
  ;[1 1 1 0.5] Overlaps
  (if (not (Variables|UserOptions|GetPickupRadius_Override))
    (CallFunction|DoPickup OtherActor)))

(event OnComponentBeginOverlap(Sphere)
  (params (OverlappedComponent :type Primitive%20Component%20Object%20Reference) (OtherActor :type Actor%20Object%20Reference) (OtherComp :type Primitive%20Component%20Object%20Reference) (OtherBodyIndex :type Integer) (bFromSweep :type Boolean) (SweepResult :type Hit%20Result%20Structure%20%28by%20ref%29))
  ;[1 1 1 0.5] Overlaps
  (if (and (Variables|UserOptions|GetPickupRadius_Override) (> (Variables|UserOptions|GetPickupRadius) 0))
    (CallFunction|DoPickup OtherActor)))

(event MinigameStateChangedDelegate(FortMinigameProgress)
  (params (Minigame :type Fort%20Minigame%20Object%20Reference) (NewMinigameState :type EFortMinigameState%20Enum :default "PreGame"))
  ;[0 0.2959 1 0.5] State Logic
  (CallFunction|UpdateForMinigameState NewMinigameState)
  (CallFunction|PlayAmbientLoopSound)
  (if (or (and (Utilities|Enum|Equal(Enum) NewMinigameState "PreGame") (Class|CreativeCommonDeviceFunctionLibrary|IsinPlayMode self)) (Utilities|Enum|Equal(Enum) NewMinigameState "WaitingForCameras"))
    (CallFunction|SetIsPowerupCollectible (Variables|UserOptions|GetSpawnOnMinigameStart) false) ; Set initial hidden state when play mode changes.
  ))

(event MinigameStartedDelegate(FortMinigameProgress)
  ;[0 0.2959 1 0.5] State Logic
  (CallFunction|SetIsPowerupCollectible (or (Variables|UserOptions|GetSpawnOnMinigameStart) (and (Variables|UserOptions|Basic|GetRespawn) (Math|Float|NearlyEqual(Float) (Variables|UserOptions|Basic|GetTimeToRespawn)))) false)
  (CallFunction|PlayAmbientLoopSound)
  (CallFunction|HandleRespawnTimer (not (Variables|UserOptions|GetSpawnOnMinigameStart))) ; If we've not spawned it initially, start the respawn timer here
  (CallFunction|BindJoininProgress true)
  (CallFunction|ApplySubscriptiontoPlayerChangesTeamorClass true))

(event MinigameEndedDelegate(FortMinigameProgress)
  ;[0 0.2959 1 0.5] State Logic
  (CallFunction|SetIsPowerupCollectible true false)
  (CallFunction|PlayAmbientLoopSound)
  (CallFunction|BindJoininProgress false)
  (CallFunction|ApplySubscriptiontoPlayerChangesTeamorClass false))

(event Custom|Player%20Join%20In%20Progress
  (params (Pawn :type Pawn%20Object%20Reference))
  ;[0 0.2959 1 0.5] State Logic
  (CallFunction|ApplySubscriptiontoPlayerChangesTeamorClass true))

(event EventEndPlay
  (params (EndPlayReason :type EEndPlayReason%20Enum :default "Destroyed"))
  ;[1 0.007234 0 0.5] Device is Removed
  (CallFunction|ApplySubscriptiontoPlayerChangesTeamorClass false))

(event EventBeginPlay
  (bind _fortminigameprogress (Variables|Default|GetFortMinigameProgress))
  (bind _currentminigame (Variables|/Script/FortniteGame.FortMinigameLogicComponent|GetCurrentMinigame _fortminigameprogress))
  ;[0.1018 1 0 0.5] Initialization
  (Networking|CanExecuteCosmeticEvents
    (:True
      (CallFunction|ClientPlayAmbienLoopSound))
    (:False
      _))
  (CallFunction|InitializeRestriction)
  (Utilities|IsValid _currentminigame
    (:Is%20Valid
      (CallFunction|UpdateForMinigameState (Variables|/Script/FortniteGame.FortMinigame|GetCurrentState _currentminigame))
      (if (Class|CreativeCommonDeviceFunctionLibrary|IsinPlayMode _currentminigame)
        (CallFunction|ApplySubscriptiontoPlayerChangesTeamorClass true)))))

(event OnPropertiesLoaded(ToyOptionsComponent)
  ;[0.1018 1 0 0.5] Initialization
  (CallFunction|UpdateforPropertyChange)
  (Networking|CanExecuteCosmeticEvents
    (:True
      (CallFunction|ClientPlayAmbienLoopSound))
    (:False
      _))
  (CallFunction|InitializeShowEffects))

(event OnAnyPropertyChanged(ToyOptionsComponent)
  ;[0.1018 1 0 0.5] Initialization
  (CallFunction|UpdateforPropertyChange)
  (Networking|CanExecuteCosmeticEvents
    (:True
      (CallFunction|ClientPlayAmbienLoopSound))
    (:False
      _))
  (CallFunction|InitializeShowEffects))

(event EventOnWorldReady
  ;[0.1018 1 0 0.5] Initialization
  (|CallParentFunction)
  (Default|UpdateVisibility))

(event EditorPreview|EventSetupForEditorPreview
  ;[0.1018 1 0 0.5] Initialization
  (|CallParentFunction)
  (Default|UpdateVisibility))

(event Custom|SpawnEvent
  (params (Instigator :type Controller%20Object%20Reference))
  ;[0.4921 0 1 0.5] Gameplay Functions
  (CallFunction|SetIsPowerupCollectible true false))

(event Custom|DespawnEvent
  (params (Instigator :type Controller%20Object%20Reference))
  ;[0.4921 0 1 0.5] Gameplay Functions
  (CallFunction|SetIsPowerupCollectible false false))

(event Custom|PickupEvent
  (params (Instigator :type Controller%20Object%20Reference))
  (bind _returnvalue (Pawn|GetControlledPawn Instigator))
  ;[0.4921 0 1 0.5] Gameplay Functions
  (Utilities|IsValid Instigator
    (:Is%20Valid
      (Utilities|IsValid _returnvalue
        (:Is%20Valid
          (CallFunction|DoPickup _returnvalue))
        (:Is%20Not%20Valid
          _)))
    (:Is%20Not%20Valid
      _))
  (CallFunction|AttemptPickupForAll))

(event Custom|ClearForAllEvent
  (params (Instigator :type Controller%20Object%20Reference))
  ;[0.4921 0 1 0.5] Gameplay Functions
  (CallFunction|RemovePowerUp (Pawn|GetControlledPawn Instigator)))

(event Custom|ClearEvent
  (params (Instigator :type Controller%20Object%20Reference))
  (bind _returnvalue (Pawn|GetControlledPawn Instigator))
  ;[0.4921 0 1 0.5] Gameplay Functions
  (Utilities|IsValid Instigator
    (:Is%20Valid
      (Utilities|IsValid _returnvalue
        (:Is%20Valid
          (CallFunction|RemovePowerUpSingular _returnvalue))))))

(event OnAddedtoMinigame(FortMinigameProgress)
  (params (Minigame :type Fort%20Minigame%20Object%20Reference))
  ;[1 0.3619 0] Update Collision Overlaps when minigame is replicated
  (CallFunction|UpdateSphereCollision (Variables|/Script/FortniteGame.FortMinigame|GetCurrentState Minigame)))

(event Custom|On%20Powerup%20Removed
  (params (For%20Controller :type Controller%20Object%20Reference)))

(event Custom|RPCClient%20PlayPickupEffects
  ;[1 1 1] Client pickup effects, only device visual effects. player vfx are aplied by the GAS system.
  (CallFunction|PlayPickupEffects))

(event Custom|RPCClientPlayPickupEffectAndResetIt
  ;[1 1 1] Client pickup effects, only device visual effects. player vfx are aplied by the GAS system.
  (CallFunction|PlayPickupEffectsandResetIt))

(fn AgentHasEffect
  (params (Agent :type Controller%20Object%20Reference))
  (bind _returnvalue (Ability|GetFortAbilitySystemComponent Agent))
  ;[0 0.787 1 0.2] Checks if there are any effects of this type on the input player, and returns if there are. Requires all Powerups to have an 'Additional Asset Tag' setup.
  (Utilities|IsValid Agent
    (:Is%20Valid
      (Utilities|IsValid _returnvalue
        (:Is%20Valid
          (bind _returnvalue_2 (GameplayEffects|GetActiveEffectswithAllTags _returnvalue (Variables|Default|GetAdditional%20Asset%20Tags)))
          (return (Math|Conversions|ToBoolean(Integer) (Utilities|Array|Length _returnvalue_2))))))))

(fn AllowInverseScaling
  (return false))

(fn ApplyPowerupEffecttoPlayer
  (params (Player%20Pawn :type Pawn%20Object%20Reference) (Force%20Disable :type Boolean))
  (bind _player_pawn (Variables|GetPlayer%20Pawn))
  (Ability|SendGameplayEventtoActor self (Variables|Default|GetPowerupAbilityActivationTag) (Utilities|Struct|MakeGameplayEventData (Variables|Default|GetPowerupAbilityActivationTag) Player%20Pawn 0 self 0 "()" "(GameplayTags=)" "(GameplayTags=)" (Math|Float|SelectFloat 0.0 1.0 (or Force%20Disable (Variables|UserOptions|GetDisablesEffectOnPickup)))))
  (if (Variables|UserOptions|GetPersistOnElimination)
    (bind _returnvalue (GameplayTags|HasMatchingGameplayTag(Message) _player_pawn "(TagName=\"Granted.Creative.Powerup.PersistOnDeathCleanup\")"))
    (if (not _returnvalue)
      (Ability|Advanced|GrantandActivateAbilityonSelfwithParams (Ability|GetAbilitySystemComponent _player_pawn) "/Game/Creative/Devices/Common/CreativePowerups/Abilites/GA_PersistOnDeath_Cleanup.GA_PersistOnDeath_Cleanup_C" 0 (Utilities|Struct|MakeGameplayEventData "(TagName=\"\")" self)))))

(fn ApplySubscriptionToPlayerChangesTeamOrClass
  (params (Subscribe :type Boolean))
  (bind _outputdelegate (EventDispatchers|CreateEvent :!event Update%20Visibility))
  (bind _fortminigameprogress (Variables|Default|GetFortMinigameProgress))
  (bind _currentminigame (Variables|/Script/FortniteGame.FortMinigameLogicComponent|GetCurrentMinigame _fortminigameprogress))
  (Networking|CanExecuteCosmeticEvents
    (:True
      (Variables|/Script/FortniteGame.FortMinigameLogicComponent|GetCurrentMinigame (Variables|Default|GetFortMinigameProgress)
        (:then
          (for _array_element (Fortnite|GetLocalFortPlayerControllers true)
            (bind _asfort_player_state_athena (Utilities|Casting|CastToFortPlayerStateAthena (Variables|/Script/Engine.Controller|GetPlayerState _array_element)))
            (if (Variables|GetSubscribe)
              (Player|BindEventtoOnTeamIndexChangedDelegate _asfort_player_state_athena _outputdelegate)
              (Minigame|BindEventtoOnAnyLocalPlayerChangedClass _currentminigame _outputdelegate)
              (else
                (Player|UnbindEventfromOnTeamIndexChangedDelegate _asfort_player_state_athena _outputdelegate)
                (Minigame|UnbindEventfromOnAnyLocalPlayerChangedClass _currentminigame _outputdelegate)))))
        (:else)) ; Valid check before attempting to use\n(on destruction Minigame may have been \ncleaned up before this)
    )))

(fn AttemptPickupforAll
  (if (Utilities|Enum|Equal(Enum) (Variables|UserOptions|GetApplyTo) "NewEnumerator4")
    (if (CreativeUserRestrictions|CanParticipantActorInteract (Variables|Components|GetCreativeUseRestriction) 0 true true false)
      (bind _currentminigame (Variables|/Script/FortniteGame.FortMinigameLogicComponent|GetCurrentMinigame (Variables|Default|GetFortMinigameProgress))
        (:then
          (bind _outplayers (Minigame|GetParticipatingPlayers_NotPure _currentminigame))
          (for _array_element _outplayers
            (Utilities|IsValid _array_element
              (:Is%20Valid
                (bind _asfort_player_pawn (Utilities|Casting|CastToFortPlayerPawn (PlayerState|GetPawn _array_element)))
                (CallFunction|ApplyPowerupEffectToPlayer _asfort_player_pawn))))
          (CallFunction|SetIsPowerupCollectible false)
          (GameplayEvent|BroadcastEvent (Variables|UserOptions|Events|GetOn%20Item%20Picked%20Up) 0 (Utilities|Struct|MakeGameplayEventPickedUp 0)))
        (:else)))))

(fn BindJoinInProgress
  (params (Bind :type Boolean))
  (bind _outputdelegate (EventDispatchers|CreateEvent :!event Player%20Join%20In%20Progress))
  (Utilities|IsValid (Variables|/Script/FortniteGame.FortMinigameLogicComponent|GetCurrentMinigame (Variables|Default|GetFortMinigameProgress))
    (:Is%20Valid
      (if (Variables|GetBind)
        (Minigame|BindEventtoOnPlayerJoininProgress (Variables|/Script/FortniteGame.FortMinigameLogicComponent|GetCurrentMinigame (Variables|Default|GetFortMinigameProgress)) _outputdelegate)
        (else
          (Minigame|UnbindEventfromOnPlayerJoininProgress (Variables|/Script/FortniteGame.FortMinigameLogicComponent|GetCurrentMinigame (Variables|Default|GetFortMinigameProgress)) _outputdelegate))))))

(fn ClearSingular
  (params (Instigator :type Controller%20Object%20Reference))
  (CallFunction|ClearEvent Instigator))

(fn DoPickup
  (params (Actor :type Actor%20Object%20Reference))
  (Networking|SwitchHasAuthority
    (:Authority
      (Utilities|IsCurrentMinigameActive
        (:Active
          (if (CallFunction|IsCollectable)
            (bind _playerpawn (CallFunction|GetPlayerPawnfromActor Actor))
            (bind _output_get (LocalVariables|SetLocalPawn _playerpawn))
            (Utilities|IsValid _output_get
              (:Is%20Valid
                (bind _valid_controller (CreativeUserRestrictions|ActorHasValidParticipantController (Variables|Components|GetCreativeUseRestriction) (Variables|GetActor) true true true))
                (if _valid_controller
                  (bind _players (Default|GetApplytoList (LocalVariables|GetLocalPawn)))
                  (for _array_element _players
                    (CallFunction|ApplyPowerupEffectToPlayer _array_element))
                  (CallFunction|SetIsPowerupCollectible false)
                  (bind _localpawn (LocalVariables|GetLocalPawn)
                    (:then
                      (LocalVariables|SetPawn%20Controller (Pawn|GetController _localpawn)))
                    (:else
                      _))
                  (GameplayEvent|BroadcastEvent (Variables|UserOptions|Events|GetOn%20Item%20Picked%20Up) 0 (Utilities|Struct|MakeGameplayEventPickedUp (LocalVariables|GetPawn%20Controller))))))))))))

(fn ForceGridAlignment
  (return false))

(fn GetEffectMagnitude
  (return 0.0))

(fn GetAdditionalSelectableActors
  (return 0))

(fn GetAgentRemainingTime
  (params (Agent :type Controller%20Object%20Reference))
  (bind _returnvalue (Ability|GetFortAbilitySystemComponent Agent))
  ;[0 0.787 1 0.2] Checks if there are any effects of this type on the input player, and returns the largest time remaining for them. Requires any powerup to have at least one 'Additional Asset Tag' to setup.
  (Utilities|IsValid Agent
    (:Is%20Valid
      (Utilities|IsValid _returnvalue
        (:Is%20Valid
          (bind _returnvalue_2 (GameplayEffects|GetActiveEffectswithAllTags _returnvalue (Variables|Default|GetAdditional%20Asset%20Tags)))
          (for _array_element _returnvalue_2
            (bind _returnvalue_3 (Ability|GameplayEffect|GetActiveGameplayEffectRemainingDuration _array_element))
            (if (== _returnvalue_3 -1.0)
              (return -1.0)
              (elif (> _returnvalue_3 (LocalVariables|GetBestTime))
                (LocalVariables|SetBestTime _returnvalue_3))))
          (return (LocalVariables|GetBestTime)))
        (:Is%20Not%20Valid
          (return 0.0))))))

(fn GetApplyToList
  (params (InstigatingPawn :type Pawn%20Object%20Reference))
  (bind _instigatingpawn (Variables|GetInstigatingPawn))
  (Utilities|Array|Clear (LocalVariables|GetApplyToListWithoutDead))
  (Utilities|FlowControl|Sequence
    (:then_0
      (bind _instigatingpawn_2 (Variables|GetInstigatingPawn)
        (:then
          (bind _fortminigameprogress (Variables|Default|GetFortMinigameProgress)
            (:then
              (bind _currentminigame (Variables|/Script/FortniteGame.FortMinigameLogicComponent|GetCurrentMinigame _fortminigameprogress)
                (:then
                  (Utilities|IsValid (Minigame|GetFNEClassComponent _currentminigame)
                    (:Is%20Valid
                      (switch Utilities|FlowControl|Switch|SwitchonEnum_Creative_Powerup_Applicability (Variables|UserOptions|GetApplyTo)
                        (:NewEnumerator0
                          (LocalVariables|SetApplyToList (Utilities|Array|MakeArray _instigatingpawn)))
                        (:NewEnumerator1
                          (bind _players (CallFunction|GetPlayersbyTeam (Team|GetActorTeam _instigatingpawn)))
                          (LocalVariables|SetApplyToList _players))
                        (:NewEnumerator2
                          (bind _players_2 (Default|GetPlayersbyClassSlot (Minigame|GetClassSlotforPlayer (Minigame|GetFNEClassComponent _currentminigame) (Utilities|Casting|CastToFortPlayerState (Variables|/Script/Engine.Pawn|GetPlayerState _instigatingpawn_2)) false)))
                          (LocalVariables|SetApplyToList _players_2))
                        (:NewEnumerator3
                          (bind _players_3 (Default|GetPlayersbyClassSlot (Minigame|GetClassSlotforPlayer (Minigame|GetFNEClassComponent _currentminigame) (Utilities|Casting|CastToFortPlayerState (Variables|/Script/Engine.Pawn|GetPlayerState _instigatingpawn_2)) false)))
                          (for _array_element _players_3
                            (if (Math|Byte|Equal(Byte) (Team|GetActorTeam _array_element) (Team|GetActorTeam (Variables|GetInstigatingPawn)))
                              (Utilities|Array|Add (LocalVariables|GetApplyToList) _array_element))))
                        (:NewEnumerator4
                          _)))))
                (:else)))
            (:else)))
        (:else
          (if (Utilities|Enum|Equal(Enum) (Variables|UserOptions|GetApplyTo) "NewEnumerator4")
            _)))
      (bind _fortminigameprogress_2 (Variables|Default|GetFortMinigameProgress)
        (:then
          (bind _currentminigame_2 (Variables|/Script/FortniteGame.FortMinigameLogicComponent|GetCurrentMinigame _fortminigameprogress_2)
            (:then
              (for _array_element_2 (Minigame|GetParticipatingPlayers _currentminigame_2)
                (Utilities|IsValid (PlayerState|GetCurrentPawn _array_element_2)
                  (:Is%20Valid
                    (Utilities|Array|Add (LocalVariables|GetApplyToList) (PlayerState|GetCurrentPawn _array_element_2))))))
            (:else)))
        (:else)))
    (:then_1
      (for _array_element_3 (LocalVariables|GetApplyToList)
        (bind _asfort_pawn (Utilities|Casting|CastToFortPawn _array_element_3))
        (if (not (FortPawn|IsDead _asfort_pawn))
          (Utilities|Array|Add (LocalVariables|GetApplyToListWithoutDead) _asfort_pawn))))
    (:then_2
      (return (LocalVariables|GetApplyToListWithoutDead)))))

(fn GetBlockPitchAndRoll
  (return false))

(fn GetDuration
  (return (Math|Float|SelectFloat (Variables|Verse|GetDurationOverride) (Math|Float|SelectFloat -1.0 (Math|Float|Clamp(Float) (Variables|UserOptions|Basic|GetEffectDuration_Option) 0.1 3e+02) (Variables|UserOptions|Basic|GetInfiniteEffectDuration)) (Variables|Verse|GetHasDurationOverride))))

(fn GetMagnitude
  ;[0 0.787 1 0.2] If there's a magnitude override, use that, otherwise use the value (generally defined by child override of the GetEffectMagnitude function)
  (return (Math|Float|SelectFloat (Variables|Verse|GetMagnitudeOverride) (CallFunction|GetEffectMagnitude) (Variables|Verse|GetHasMagnitudeOverride))))

(fn GetMagnitudeMinMax
  (return 0.0 65535.0))

(fn GetMagnitudeTags
  (return (Utilities|Map|MakeMap (Variables|Default|GetMagnitudeTag) (CallFunction|GetMagnitude))))

(fn GetPlayerPawnFromActor
  (params (Actor :type Actor%20Object%20Reference))
  (bind _asplayer_pawn_athena (Utilities|Casting|CastToPlayerPawn_Athena Actor)
    (:then
      (return _asplayer_pawn_athena))
    (:CastFailed
      (bind _asfort_athena_vehicle (Utilities|Casting|CastToFortAthenaVehicle Actor)
        (:then
          (Utilities|IsValid (Vehicle|GetDriver _asfort_athena_vehicle)
            (:Is%20Valid
              (return (Vehicle|GetDriver _asfort_athena_vehicle)))
            (:Is%20Not%20Valid
              _)))
        (:CastFailed
          _))
      (return 0))))

(fn GetPlayersByClassSlot
  (params (ClassSlot :type Byte))
  (Utilities|Array|Clear (LocalVariables|GetSameClassSlot%20Players))
  (bind _fortminigameprogress (Variables|Default|GetFortMinigameProgress)
    (:then
      (bind _currentminigame (Variables|/Script/FortniteGame.FortMinigameLogicComponent|GetCurrentMinigame _fortminigameprogress)
        (:then
          (Utilities|IsValid (Minigame|GetFNEClassComponent _currentminigame)
            (:Is%20Valid
              (LocalVariables|SetLocalMinigameFNEClassComponent (Minigame|GetFNEClassComponent _currentminigame))
              (for _array_element (Minigame|GetParticipatingPlayers _currentminigame)
                (if (Math|Byte|Equal(Byte) (Minigame|GetClassSlotforPlayer (LocalVariables|GetLocalMinigameFNEClassComponent) _array_element false) (Variables|GetClassSlot))
                  (Utilities|Array|Add (LocalVariables|GetSameClassSlot%20Players) (Variables|/Script/Engine.PlayerState|GetPawnPrivate _array_element))))
              (return (LocalVariables|GetSameClassSlot%20Players)))
            (:Is%20Not%20Valid
              _)))
        (:else
          _))
      (return (LocalVariables|GetSameClassSlot%20Players)))
    (:else)))

(fn GetPlayersByTeam
  (params (RequiredTeamIndex :type Byte))
  (for _array_element (Minigame|GetParticipatingPlayers (Variables|/Script/FortniteGame.FortMinigameLogicComponent|GetCurrentMinigame (Variables|Default|GetFortMinigameProgress)))
    (if (Math|Byte|Equal(Byte) (Team|GetActorTeam _array_element) (Variables|GetRequiredTeamIndex))
      (Utilities|Array|Add (LocalVariables|GetSameTeamPlayers) (Utilities|Casting|CastToPlayerPawn_Athena (PlayerState|GetCurrentPawn _array_element)))))
  (return (LocalVariables|GetSameTeamPlayers)))

(fn GetScaleMax
  (params (DefaultScaleMax :type Float%20%28single-precision%29))
  (return DefaultScaleMax))

(fn GetScaleMin
  (params (DefaultScaleMin :type Float%20%28single-precision%29))
  (return DefaultScaleMin))

(fn GetShowFXOnPlayer
  (return (Variables|Default|GetbPlayFXOnPlayer)))

(fn GetSpawnedActors
  (return 0))

(fn HandleRespawnTimer
  (params (Start :type Boolean))
  (Networking|SwitchHasAuthority
    (:Authority
      (if Start
        (bind _returnvalue (Utilities|Time|SetTimerbyEvent (EventDispatchers|CreateEvent :!event ResetPowerupCollectible_TimerDelegate) (Math|Float|SelectFloat (Variables|UserOptions|Basic|GetTimeToRespawn) -1.0 (Variables|UserOptions|Basic|GetRespawn))))
        (Variables|Default|SetRespawnTimerHandle _returnvalue)
        (else
          (Utilities|Time|ClearandInvalidateTimerbyHandle (Variables|Default|GetRespawnTimerHandle)))))))

(fn HandleFX
  (Rendering|Material|SetMaterial (Variables|Components|GetFX) 0 0))

(fn HasAdditionalSelectableActors
  (return false))

(fn HasAdditionalValidPlacementLogic
  ;[1 0.8399 0.1063] By default, powerups don't need additional placement logic (i.e. return false so IsValidPlacement will not be called).
  (return false))

(fn InitializeShowEffects
  (bind _showfxgameplaytag (Variables|Default|GetShowFXGameplayTag))
  (bind _returnvalue (GameplayTags|MakeGameplayTagContainerfromTag _showfxgameplaytag))
  ;[1 0.6099 0 0.2] Applies a tag to the device that will be read by the Gameplay Cue to determine if it should show effects or not.
  (if (CallFunction|GetShowFXOnPlayer)
    (if (not (GameplayTags|HasMatchingGameplayTag self (Variables|Default|GetShowFXGameplayTag)))
      (Ability|GameplayEffect|AddLooseGameplayTags self _returnvalue))
    (else
      (Ability|GameplayEffect|RemoveLooseGameplayTags self _returnvalue))))

(fn InitializeRestriction
  (Initialization|Initialize (Variables|Components|GetCreativeUseRestriction) (Variables|Default|GetFortMinigameProgress) 0 0 (Variables|UserOptions|GetSelectedTeam) false (Variables|UserOptions|GetSelectedClass) false false))

(fn IsCollectable
  (return (Utilities|Array|Get(acopy) (Variables|Default|GetIsPowerupCollectible)) (Utilities|Array|Get(acopy) (Variables|Default|GetIsPowerupCollectible) 1)))

(fn IsSpawned
  ;[0 0.787 1 0.2] Returns if the powerup is currently collectible.
  (return (CallFunction|IsCollectable)))

(fn IsValidPlacement
  ;[1 0.9588 0.2402] Will not be called by default because HasAdditionalValidPlacementLogic returns false by default.
  (return false))

(fn MakeTagIndexArray
  (params (%5B0%5D :type Integer) (%5B1%5D :type Integer) (%5B2%5D :type Integer))
  ;[0 0.7015 1 0.25] Returns the array of tags, but if the previous tag is not set, we want to ignore the current tag. This is because we hide those from the player.
  (return (Utilities|Array|MakeArray %5B0%5D (Math|Integer|SelectInt %5B1%5D 0 (!= %5B0%5D 0)) (Math|Integer|SelectInt %5B2%5D 0 (!= (Math|Integer|SelectInt %5B1%5D 0 (!= %5B0%5D 0)) 0)))))

(fn MigrateLegacyUserOptions
  (bind _effectduration (Class|GetClassDefaults "/CreativeCoreDevices/Creative_Powerup_V2.Creative_Powerup_V2_C"))
  (bind _respawntime (Class|GetClassDefaults "/CreativeCoreDevices/Creative_Powerup_V2.Creative_Powerup_V2_C"))
  (bind _respawntime_2 (Variables|Legacy|GetRespawnTime))
  (bind _returnvalue (!= _respawntime_2 -1.0))
  (bind _effectduration_2 (Variables|Legacy|GetEffectDuration))
  (|CallParentFunction)
  (Utilities|FlowControl|Sequence
    (:then_0
      ;[1 1 1] EffectDuration -> InfiniteEffectDuration , EffectDuration_Option
      (if (!= _effectduration (Variables|Legacy|GetEffectDuration))
        (Variables|UserOptions|Basic|SetInfiniteEffectDuration (== _effectduration_2 1.5e+04))
        (Variables|UserOptions|Basic|SetEffectDuration_Option _effectduration_2)
        (Variables|Legacy|SetEffectDuration _effectduration)
        (LocalVariables|SetbOptionsMigrated true)))
    (:then_1
      ;[1 1 1] RespawnTime -> Respawn , TimeToRespawn
      (if (!= _respawntime (Variables|Legacy|GetRespawnTime))
        (Variables|UserOptions|Basic|SetRespawn _returnvalue)
        (Variables|UserOptions|Basic|SetTimeToRespawn (Math|Float|SelectFloat (Variables|Legacy|GetRespawnTime) (Class|GetClassDefaults "/CreativeCoreDevices/Creative_Powerup_V2.Creative_Powerup_V2_C") _returnvalue))
        (Variables|Legacy|SetRespawnTime _respawntime)
        (LocalVariables|SetbOptionsMigrated true)))
    (:then_2
      ;[1 1 1] Pickup Radius
      (if (< (Variables|UserOptions|GetPickupRadius) 0)
        (Variables|UserOptions|SetPickupRadius_Override false)
        (Variables|UserOptions|SetPickupRadius 0.25)
        (LocalVariables|SetbOptionsMigrated true)))
    (:then_3
      (if (LocalVariables|GetbOptionsMigrated)
        (OptionData|RefreshPropertyOverrides (Variables|Components|GetToyOptionsComponent))))))

(fn OnPreviewEnd
  ;[1 0.7418 0.108] Keep track of if this actor is the current move tool preview actor.
  (Variables|Default|SetbIsCreativeMoveToolPreview false)
  (return false))

(fn OnPreviewStart
  ;[0.4802 1 0.3444] Keep track of if this actor is the current move tool preview actor.
  (Variables|Default|SetbIsCreativeMoveToolPreview true)
  (return false))

(fn OnRep_IsPowerupCollectible
  (Networking|CanExecuteCosmeticEvents
    (:True
      (Default|UpdateVisibility)
      ;[1 1 1] if it can be picked up, I reset the variable so that the visual effect can be displayed again for any reason.
      (if (CallFunction|IsCollectable)
        (CallFunction|ResetPlayCollectionFXGate)))
    (:False
      (Networking|SwitchHasAuthority
        (:Authority
          (Networking|ForceNetUpdate))))))

(fn PlayPickupEffects
  (Networking|CanExecuteCosmeticEvents
    (:True
      (CallFunction|PlayCollectionFX)
      (CallFunction|PlayPickUpSound)
      (CallFunction|FadeOutAmbientSound))))

(fn PlayPickupEffectsAndResetIt
  (Networking|CanExecuteCosmeticEvents
    (:True
      (CallFunction|PlayCollectionFX)
      (CallFunction|PlayPickUpSound)
      (CallFunction|ResetPlayCollectionFXGate))))

(fn RemovePowerUp
  (params (Pawn :type Pawn%20Object%20Reference))
  (Networking|SwitchHasAuthority
    (:Authority
      (Utilities|IsCurrentMinigameActive
        (:Active
          (bind _players (Default|GetApplytoList Pawn))
          (for _array_element _players
            (CallFunction|ApplyPowerupEffectToPlayer _array_element true)))))))

(fn RemovePowerUpSingular
  (params (Actor :type Actor%20Object%20Reference))
  (Networking|SwitchHasAuthority
    (:Authority
      (Utilities|IsCurrentMinigameActive
        (:Active
          (bind _playerpawn (CallFunction|GetPlayerPawnfromActor Actor))
          (bind _output_get (LocalVariables|SetLocal%20Pawn _playerpawn))
          (Utilities|IsValid _output_get
            (:Is%20Valid
              (bind _valid_controller (CreativeUserRestrictions|ActorHasValidParticipantController (Variables|Components|GetCreativeUseRestriction) (Variables|GetActor) true true true))
              (if _valid_controller
                (CallFunction|ApplyPowerupEffectToPlayer (LocalVariables|GetLocal%20Pawn) true)))))))))

(fn ResetOverrides
  (Variables|Verse|SetHasDurationOverride false)
  (Variables|Verse|SetHasMagnitudeOverride false)
  (Variables|Default|SetIsPowerupCollectible (Utilities|Array|MakeArray true false)))

(fn ResetPowerupCollectible_TimerDelegate
  (CallFunction|SetIsPowerupCollectible true)
  (CallFunction|PlayAmbientLoopSound))

(fn SetDurationOverride
  (params (NewDuration :type Float%20%28double-precision%29))
  (Variables|Verse|SetDurationOverride (Math|Float|Max(Float) NewDuration))
  (Variables|Verse|SetHasDurationOverride true))

(fn SetIsPowerupCollectible
  (params (IsCollectible :type Boolean) (Has%20Been%20Picked%20Up :type Boolean :default true))
  (bind (_is_collectable _has_been_picked_up) (CallFunction|IsCollectable))
  (Networking|SwitchHasAuthority
    (:Authority
      (if (Math|Boolean|NotEqual(Boolean) (CallFunction|IsCollectable) IsCollectible)
        (if (or (!= (Math|Float|SelectFloat (Variables|UserOptions|Basic|GetTimeToRespawn) -1.0 (Variables|UserOptions|Basic|GetRespawn)) 0) (not (Variables|GetHas%20Been%20Picked%20Up)))
          (Variables|Default|SetIsPowerupCollectible (Utilities|Array|MakeArray (Variables|GetIsCollectible) (Variables|GetHas%20Been%20Picked%20Up))) ; 0 = Is Collectable\n1 = Play FX
          (if (and (not (Variables|GetIsCollectible)) (Variables|GetHas%20Been%20Picked%20Up))
            (CallFunction|RPCClientPlayPickupEffects))
          (CallFunction|HandleRespawnTimer (not (or _is_collectable (not _has_been_picked_up))))
          (else
            ;[1 1 1] there is no valid respawning, so  only show the pickup effects
            (CallFunction|RPCClientPlayPickupEffectandResetIt)))))))

(fn SetMagnitudeOverride
  (params (NewMagnitude :type Float%20%28double-precision%29))
  (bind (_min _max) (CallFunction|GetMagnitudeMinMax))
  (Variables|Verse|SetMagnitudeOverride (Math|Float|Clamp(Float) NewMagnitude _min _max))
  (Variables|Verse|SetHasMagnitudeOverride true))

(fn SetupEffectData
  (params (EffectHandle :type Gameplay%20Effect%20Spec%20Handle%20Structure) (EffectClass :type Gameplay%20Effect%20Class%20Reference))
  (bind _returnvalue (Ability|GameplayEffect|SetDuration EffectHandle (CallFunction|GetDuration)))
  (bind _output_get (LocalVariables|SetLocalMagnitudeTags (CallFunction|GetMagnitudeTags)))
  (bind _keys (Utilities|Map|Keys _output_get))
  (for _array_element _keys
    (Ability|GameplayEffect|AssignTagSetbyCallerMagnitude _returnvalue _array_element (Utilities|Map|Find _output_get _array_element)))
  (bind _returnvalue_2 (Ability|GameplayEffect|AddGrantedTags _returnvalue (Variables|Default|GetAdditional%20Tags)))
  (bind _returnvalue_3 (Ability|GameplayEffect|AddAssetTags _returnvalue_2 (Variables|Default|GetAdditional%20Asset%20Tags)))
  (if (Variables|UserOptions|GetPersistOnElimination)
    (bind _returnvalue_4 (Ability|GameplayEffect|AddAssetTags _returnvalue_3 "(GameplayTags=((TagName=\"Asset.Creative.CreativeEffect.PersistOnDeath\"),(TagName=\"Gameplay.Effect.PersistOnDeath\")))")) ; Gameplay.Effect.PersistOnDeath is a special tag applied to effects that won't wear off on death.\nAsset.Creative.CreativeEffect.PersistOnDeath is added to allow us to gather all the effects that persist on death at the end of the minigame\nand remove them (since they won't auto-remove on the players despawning).
    (return _returnvalue_4 true)
    (else
      (return _returnvalue_3 true))))

(fn ShouldBlockPlacement
  (params (MoveTool :type Fort%20Creative%20Move%20Tool%20Object%20Reference))
  (return false))

(fn UpdateforMinigameState
  (params (MinigameState :type EFortMinigameState%20Enum :default "PreGame"))
  (CallFunction|UpdateSphereCollision MinigameState)
  (Default|UpdateVisibility))

(fn UpdateVisibility
  (Networking|CanExecuteCosmeticEvents
    (:True
      (if (Class|CreativeCommonDeviceFunctionLibrary|IsLogicinPlayMode (Variables|Default|GetFortMinigameProgress))
        (if (CallFunction|IsCollectable)
          (CallFunction|UpdateinPlayVisibility)
          (else
            (Rendering|SetActorHiddenInGame true)))
        (else
          (Rendering|SetActorHiddenInGame (and (Editor|InEditor) (Utilities|Enum|NotEqual(Enum) (Variables|UserOptions|GetWhoCanSeeThisPowerup) "NewEnumerator1"))))))))

(fn UpdateForPropertyChange
  (Effects|Components|ParticleSystem|SetColorParameter (Variables|Components|GetFX) "Color" (Variables|Default|GetVFX%20Particle%20Color))
  (UEFNConversionHelper|SetComponentHiddenEditor (Variables|Components|GetDisableMesh) (not (Variables|UserOptions|GetDisablesEffectOnPickup)))
  (CallFunction|InitializeRestriction)
  (Default|UpdateVisibility)
  ;[1 1 1] Update Collision Overlaps on Property Load
  (bind _fortminigameprogress (Variables|Default|GetFortMinigameProgress)
    (:then
      (bind _currentminigame (Variables|/Script/FortniteGame.FortMinigameLogicComponent|GetCurrentMinigame _fortminigameprogress)
        (:then
          (CallFunction|UpdateSphereCollision (Variables|/Script/FortniteGame.FortMinigame|GetCurrentState _currentminigame)))
        (:else)))
    (:else)))

(fn UpdateInPlayVisibility
  (switch Utilities|FlowControl|Switch|SwitchonEnum_Creative_Powerup_Visibility (Variables|UserOptions|GetWhoCanSeeThisPowerup)
    (:NewEnumerator0
      (Rendering|SetActorHiddenInGame true))
    (:NewEnumerator1
      (Rendering|SetActorHiddenInGame false))
    (:NewEnumerator2
      (for _array_element (Fortnite|GetLocalFortPlayerControllers true)
        (Rendering|SetActorHiddenInGame (not (Class|CreativeUseRestrictionComponent|StateChecks (Variables|Components|GetCreativeUseRestriction) _array_element true)))))))

(fn UpdateSphereCollision
  (params (MinigameState :type EFortMinigameState%20Enum :default "PreGame"))
  (bind _staticmeshcomponent (Variables|Components|GetStaticMeshComponent))
  (bind _sphere (Variables|Components|GetSphere))
  (if (and (Variables|UserOptions|GetPickupRadius_Override) (Utilities|Enum|Equal(Enum) MinigameState "InProgress"))
    (Components|Sphere|SetSphereRadius (Variables|Components|GetSphere) (* (Variables|UserOptions|GetPickupRadius) 1e+02) true)
    (Collision|SetCollisionEnabled _staticmeshcomponent "NoCollision")
    (Collision|SetCollisionEnabled _sphere "QueryOnly")
    (else
      (Collision|SetCollisionEnabled _staticmeshcomponent "QueryOnly")
      (Collision|SetCollisionEnabled _sphere "NoCollision"))))

(fn ConstructionScript)
