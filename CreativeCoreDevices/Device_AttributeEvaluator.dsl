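; Creative Attribute Evaluator device that tests an instigating player against team, class, elimination, player score and team score requirements.
; EvaluateMatch returns a result code, using a tracked stat filter cached from minigame stats, which selects the On Pass or On Fail event broadcast.
; Successful transmits replicate the result, driving timed pass/fail mesh feedback through custom primitive data, which also shows enabled state.
; A non-cosmetic-side timer resets the replicated result afterwards.
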(event EventInitializeDevice
  (|CallParentFunction)
  (CallFunction|CacheCustomStatFilter))

(event PostMinigameStartedDelegate(FortMinigameLogic)
  (CallFunction|CacheCustomStatFilter))

(event OnAddedtoMinigame(FortMinigameLogic)
  (params (Minigame :type Fort%20Minigame%20Object%20Reference))
  (Variables|Default|SetCachedMinigameStatsComponent (Minigame|Stats|GetMinigameStatsComponent Minigame))
  (Default|SetMaterialBasedOnCurrentState))

(event EventMinigameStateChanged
  (params (New%20State :type EFortMinigameState%20Enum :default "PreGame"))
  (|CallParentFunction "PreGame")
  (if (or (Utilities|Enum|Equal(Enum) New%20State "PreGame") (Utilities|Enum|Equal(Enum) New%20State "Setup"))
    (Default|SetMaterialBasedOnCurrentState)))

(event Custom|EvaluateChecksFromEvent
  (params (Instigator :type Controller%20Object%20Reference))
  (Utilities|IsValid Instigator
    (:Is%20Valid
      (bind _output_get (Variables|Default|SetCachedMatchResult (CallFunction|EvaluateMatch (CreativeAI|Controller|GetParticipantControllerfromActor Instigator false))))
      (CallFunction|StartActivateAttempt Instigator (select (== _output_get (Variables|Default|GetSuccessMatchResultCode)) (delegate Custom|BroadcastChecksWereValid) (delegate Custom|BroadcastChecksWereInvalid))))))

(event Custom|BroadcastChecksWereInvalid
  (params (Instigator :type Controller%20Object%20Reference))
  (GameplayEvent|BroadcastEvent (Variables|UserOptions|Events|GetOn%20Fail) 0 (Utilities|Struct|MakeGameplayEventTriggeronNoMatch Instigator)))

(event Custom|BroadcastChecksWereValid
  (params (Instigator :type Controller%20Object%20Reference))
  (GameplayEvent|BroadcastEvent (Variables|UserOptions|Events|GetOn%20Pass) 0 (Utilities|Struct|MakeGameplayEventTriggeronMatch Instigator)))

(event Default|EventSetMaterialBasedOnCurrentState
  (bind (_disablecolorvfx_r _disablecolorvfx_g _disablecolorvfx_b _disablecolorvfx_a) (Math|Color|BreakColor (Variables|GetDisableColorVFX_R)))
  (bind (_enablecolorvfx_r _enablecolorvfx_g _enablecolorvfx_b _enablecolorvfx_a) (Math|Color|BreakColor (Variables|GetEnableColorVFX_R)))
  (bind (_delaycolorvfx_r _delaycolorvfx_g _delaycolorvfx_b _delaycolorvfx_a) (Math|Color|BreakColor (Variables|GetDelayColorVFX_R)))
  (bind _mesh (Variables|Default|GetMesh))
  (Networking|CanExecuteCosmeticEvents
    (:True
      (switch Utilities|FlowControl|Switch|SwitchonEEnabledState (Variables|Default|GetCurrentState)
        (:NewEnumerator0
          (Rendering|Material|SetCustomPrimitiveDataFloat _mesh (Variables|VFX|GetLightFXBlinkDataIndex) 0.0)
          (Rendering|Material|SetCustomPrimitiveDataVector4 (Variables|Default|GetMesh) (Variables|VFX|GetLightFXColorDataIndex) (Math|Vector4|MakeVector4 _enablecolorvfx_r _enablecolorvfx_g _enablecolorvfx_b _enablecolorvfx_a))
          (Rendering|Material|SetCustomPrimitiveDataFloat (Variables|Default|GetMesh) (Variables|VFX|GetEnabledFXDataIndex) (CallFunction|GetEnabledVisualState)))
        (:NewEnumerator1
          (Rendering|Material|SetCustomPrimitiveDataFloat _mesh (Variables|VFX|GetLightFXBlinkDataIndex) 1.0)
          (Rendering|Material|SetCustomPrimitiveDataVector4 (Variables|Default|GetMesh) (Variables|VFX|GetLightFXColorDataIndex) (Math|Vector4|MakeVector4 _delaycolorvfx_r _delaycolorvfx_g _delaycolorvfx_b _delaycolorvfx_a)))
        (:NewEnumerator2
          (Rendering|Material|SetCustomPrimitiveDataVector4 (Variables|Default|GetMesh) (Variables|VFX|GetLightFXColorDataIndex) (Math|Vector4|MakeVector4 _disablecolorvfx_r _disablecolorvfx_g _disablecolorvfx_b _disablecolorvfx_a))
          (Rendering|Material|SetCustomPrimitiveDataFloat (Variables|Default|GetMesh) (Variables|VFX|GetEnabledFXDataIndex) (CallFunction|GetEnabledVisualState)))))))

(event Custom|MatchResultFeedbackStart
  (CallFunction|UpdateChannelNumberwithMatchResult (Variables|Replicated|GetLastMatchResult)))

(event OnChannelIdsReplicated(ListenToChannel)
  (CallFunction|ShowListenChannel))

(event Custom|CleanupLastResult
  (Variables|Replicated|SetLastMatchResult (Variables|Default|GetInvalidMatchResultCode))
  (Utilities|Time|InvalidateTimerHandle (Variables|Default|GetCleanupTimerHandle)))

(event EventDestroyed
  (Networking|CanExecuteCosmeticEvents
    (:False
      (Utilities|Time|ClearandInvalidateTimerbyHandle (Variables|Default|GetCleanupTimerHandle)))))

(event ChannelNumberUpdate|EventUpdateChannelNumber
  (params (Num :type Integer)))

(event EventSwitchtoDyamicMaterial
  (Networking|CanExecuteCosmeticEvents
    (:True
      (if (== (Variables|Replicated|GetLastMatchResult) (Variables|Default|GetInvalidMatchResultCode))
        (CallFunction|ShowListenChannel)
        (else
          (CallFunction|MatchResultFeedbackStart))))))

(event EventSetCurrentConstantMaterial
  (params (NewCurrentMaterial :type Material%20Interface%20Object%20Reference)))

(fn CacheCustomStatFilter
  (bind (_statevent _targettagsquery _sourcetagsquery _contexttagsquery) (Utilities|Struct|BreakStatEventFilter (Variables|UserOptions|GetTrackedStat)))
  (bind _self self)
  (bind _returnvalue (Minigame|Stats|GetMinigameStatsComponent _self))
  (bind _output_get (LocalVariables|SetFoundStats (Variables|/Script/FortniteGame.FortMinigameStatsComponent|GetLoadedStats _returnvalue)))
  (Utilities|Array|AppendArray _output_get (Variables|/Script/FortniteGame.FortMinigameStatsComponent|GetPlayerDefinedStatInstances _returnvalue))
  (for _array_element _output_get
    (if (Math|Boolean|ANDBoolean (Utilities|Enum|Equal(Enum) (Variables|/Script/FortniteGame.FortMinigameStatFilter|GetStatEvent _array_element) _statevent) (or (GameplayTags|IsTagQueryEmpty (Variables|/Script/FortniteGame.FortMinigameStatFilter|GetTargetTagsQuery _array_element)) (GameplayTags|DoesContainerMatchTagQuery _targettagsquery (Variables|/Script/FortniteGame.FortMinigameStatFilter|GetTargetTagsQuery _array_element))) (or (GameplayTags|IsTagQueryEmpty (Variables|/Script/FortniteGame.FortMinigameStatFilter|GetSourceTagsQuery _array_element)) (GameplayTags|DoesContainerMatchTagQuery _sourcetagsquery (Variables|/Script/FortniteGame.FortMinigameStatFilter|GetSourceTagsQuery _array_element))) (or (GameplayTags|IsTagQueryEmpty (Variables|/Script/FortniteGame.FortMinigameStatFilter|GetContextTagsQuery _array_element)) (GameplayTags|DoesContainerMatchTagQuery _contexttagsquery (Variables|/Script/FortniteGame.FortMinigameStatFilter|GetContextTagsQuery _array_element))))
      (bind _output_get_2 (Variables|Default|SetTracked%20Stat%20Filter _array_element))
      (Variables|Default|SetStatDivisor (Math|Integer|SelectInt 1000 1 (Utilities|Set|ContainsItem (Variables|Default|GetTimeStats) (Utilities|ResolveSoftReference (Utilities|GetClass _output_get_2))))) ; time based stats are in MS. This divisor will convert to seconds.
      (return)))
  (Variables|Default|SetTracked%20Stat%20Filter 0)
  (return))

(fn EvaluateMatch
  (params (ParticipantController :type Controller%20Object%20Reference))
  (bind _localcurrentminigame (LocalVariables|GetLocalCurrentMinigame))
  (bind _returnvalue (Minigame|GetMinigameTeamComponent _localcurrentminigame))
  (bind _activatingteam (Variables|UserOptions|Basic|GetActivatingTeam))
  (Utilities|FlowControl|Sequence
    (:then_0
      (if (Creative|TeamOption|IsTeamType _activatingteam "Any")
        _
        (elif (== (Math|Byte|MakeLiteralByte (Creative|TeamOption|ConverttoTeam _activatingteam)) (Team|GetActorTeam (Variables|GetParticipantController)))
          _
          (elif (Variables|UserOptions|GetInvert%20Team%20Selection)
            _
            (else
              (return 1)))))
      (if (Variables|UserOptions|GetInvert%20Team%20Selection)
        _))
    (:then_1
      (bind _fortminigamelogic (Variables|Default|GetFortMinigameLogic)
        (:then
          (bind _currentminigame (Variables|/Script/FortniteGame.FortMinigameLogicComponent|GetCurrentMinigame _fortminigamelogic)
            (:then
              (bind _output_get (LocalVariables|SetLocalCurrentMinigame _currentminigame))
              (Utilities|IsValid (Minigame|GetFNEClassComponent _output_get)
                (:Is%20Valid
                  (LocalVariables|SetLocalMinigameFNEClassComponent (Minigame|GetFNEClassComponent _output_get))
                  (Utilities|IsValid _returnvalue
                    (:Is%20Valid
                      (LocalVariables|SetLocalMinigameTeamComponent _returnvalue)
                      (bind _asfort_player_state (Utilities|Casting|CastToFortPlayerState (Variables|/Script/Engine.Controller|GetPlayerState (Variables|GetParticipantController)))
                        (:then
                          (LocalVariables|SetLocalFortPlayerState _asfort_player_state)
                          (Variables|Default|GetCachedMinigameStatsComponent
                            (:then)
                            (:else
                              (return (Variables|Default|GetInvalidMatchResultCode)))))
                        (:CastFailed
                          _)))
                    (:Is%20Not%20Valid
                      _))
                  (return (Variables|Default|GetInvalidMatchResultCode)))
                (:Is%20Not%20Valid
                  _)))
            (:else
              (return (Variables|Default|GetInvalidMatchResultCode)))))
        (:else
          (return (Variables|Default|GetInvalidMatchResultCode)))))
    (:then_2
      (if (Creative|ClassOption|IsClassType (Variables|UserOptions|Basic|GetActivatingClass) "Any")
        _
        (elif (Math|Byte|Equal(Byte) (Minigame|GetClassSlotforPlayer (LocalVariables|GetLocalMinigameFNEClassComponent) (LocalVariables|GetLocalFortPlayerState) true) (Creative|ClassOption|ConverttoClassSlotIndex (Variables|UserOptions|Basic|GetActivatingClass)))
          _
          (elif (Variables|UserOptions|GetInvert%20Class%20Selection)
            _
            (else
              (return 2)))))
      (if (Variables|UserOptions|GetInvert%20Class%20Selection)
        _))
    (:then_3
      (if (!= (Variables|UserOptions|GetMinPlayerEliminations) 0)
        (if (< (Minigame|Stats|GetStatforPlayer (Variables|Default|GetCachedMinigameStatsComponent) (LocalVariables|GetLocalFortPlayerState) (Class|FortMinigameStatsComponent|GetLoadedStatfromClass (Variables|Default|GetCachedMinigameStatsComponent) "/Game/Athena/Playset/Minigames/StatFilters/MinigameStat_PlayerKills.MinigameStat_PlayerKills_C") false) (Variables|UserOptions|GetMinPlayerEliminations))
          (return 3))))
    (:then_4
      (bind _tracked_stat_filter (Variables|Default|GetTracked%20Stat%20Filter)
        (:then
          (if (< (/ (Minigame|Stats|GetStatforPlayer (Variables|Default|GetCachedMinigameStatsComponent) (LocalVariables|GetLocalFortPlayerState) _tracked_stat_filter false) (Variables|Default|GetStatDivisor)) (Variables|UserOptions|GetMinPlayerScore))
            (return 4)))
        (:else)))
    (:then_5
      (bind _tracked_stat_filter_2 (Variables|Default|GetTracked%20Stat%20Filter)
        (:then
          (if (== (Minigame|Teams|GetMinigameNumTeams (LocalVariables|GetLocalMinigameTeamComponent)) (LocalVariables|GetCooperative))
            (if (< (/ (Minigame|Stats|GetStatforGroup (Variables|Default|GetCachedMinigameStatsComponent) _tracked_stat_filter_2 false) (Variables|Default|GetStatDivisor)) (Variables|UserOptions|GetMinTeamScore))
              (return 5))
            (elif (< (/ (Minigame|Stats|GetStatforPlayerBucket (Variables|Default|GetCachedMinigameStatsComponent) (Minigame|GetBucketforPlayerState (LocalVariables|GetLocalMinigameTeamComponent) (LocalVariables|GetLocalFortPlayerState) true) _tracked_stat_filter_2 false false) (Variables|Default|GetStatDivisor)) (Variables|UserOptions|GetMinTeamScore))
              (return 5))))
        (:else)))
    (:then_6
      (return (Variables|Default|GetSuccessMatchResultCode)))))

(fn GetEnabledVisualState
  (bind _currentminigame (Variables|/Script/FortniteGame.FortMinigameLogicComponent|GetCurrentMinigame (Variables|Default|GetFortMinigameLogic))
    (:then
      (if (Minigame|IsinPlayMode _currentminigame :!class /Script/FortniteGame.MinigameBlueprintLibrary)
        (return (Math|Conversions|ToFloat(Boolean) (Utilities|Enum|Equal(Enum) (Variables|Default|GetCurrentState) "NewEnumerator0")))
        (else
          (return 1.0))))
    (:else
      (return 1.0))))

(fn GetFeedbackNumberByMatchResultCode
  (params (MatchResultCode :type Integer))
  (if (== (Variables|GetMatchResultCode) (Variables|Default|GetSuccessMatchResultCode))
    (return (Class|FortGameplayMessageComponentBase|GetMessageChannelIndex (Variables|Default|GetMatchTransmitOn)))
    (elif (== (Variables|GetMatchResultCode) (Variables|Default|GetInvalidMatchResultCode))
      (return (Class|FortGameplayMessageComponentBase|GetMessageChannelIndex (Variables|Default|GetListenToChannel)))
      (else
        (return (Variables|GetMatchResultCode))))))

(fn InitializeComponents
  ;[0 0.6223 1 0.25] Overriding but not initializing UseRestriction fully (since we don't use that for this child).
  (CallFunction|UpdateVisibilityDuringPhase)
  (Initialization|Initialize (Variables|Default|GetCreative_UseRestriction_Component) 0 0 0 (Utilities|Struct|MakeCreativeTeamOption "Any") false (Utilities|Struct|MakeCreativeClassOption "Any") false false))

(fn OnRep_LastMatchResult
  (Networking|CanExecuteCosmeticEvents
    (:True
      (if (== (Variables|Replicated|GetLastMatchResult) (Variables|Default|GetInvalidMatchResultCode))
        (CallFunction|ShowListenChannel)
        (else
          (CallFunction|MatchResultFeedbackStart))))
    (:False
      (Utilities|FlowControl|Sequence
        (:then_0
          (if (Utilities|Time|IsValidTimerHandle (Variables|Default|GetCleanupTimerHandle))
            (Utilities|Time|ClearandInvalidateTimerbyHandle (Variables|Default|GetCleanupTimerHandle))))
        (:then_1
          (bind _returnvalue (Utilities|Time|SetTimerbyEvent (EventDispatchers|CreateEvent :!event CleanupLastResult) (Math|Float|Max(Float) (* (Variables|Replication|GetNetUpdateFrequency) 1.5) (Variables|VFX|GetMatchResultFeedbackTime))))
          (Variables|Default|SetCleanupTimerHandle _returnvalue))))))

(fn ShowListenChannel
  (CallFunction|UpdateChannelNumberwithMatchResult (Variables|Default|GetInvalidMatchResultCode)))

(fn TransmitEvery
  (return (Variables|UserOptions|GetTransmit%20Every%20X%20Triggers)))

(fn TransmitMessage
  (bind _success (|CallParentFunction))
  (if _success
    (Variables|Replicated|SetLastMatchResult (Variables|Default|GetCachedMatchResult)))
  (return _success))

(fn UpdateChannelNumberWithMatchResult
  (params (MatchResultCode :type Integer))
  (if (== (Variables|GetMatchResultCode) (Variables|Default|GetInvalidMatchResultCode))
    (Rendering|Material|SetCustomPrimitiveDataVector4 (Variables|Default|GetMesh) (Variables|VFX|GetLineFXDataIndex) (0.0 0.0 0.0 0.0))
    (else
      (Rendering|Material|SetCustomPrimitiveDataVector4 (Variables|Default|GetMesh) (Variables|VFX|GetLineFXDataIndex) (Math|Vector4|MakeVector4 (Utilities|Time|GetTimeSeconds) (+ (Utilities|Time|GetTimeSeconds) (Variables|VFX|GetMatchResultFeedbackTime)) (Variables|VFX|GetLineFadeTime) (select (== (Variables|GetMatchResultCode) (Variables|Default|GetSuccessMatchResultCode)) 1.0 -1.0))))))

(fn ConstructionScript)
