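; Creative ball spawner device that spawns a configurable physics ball (CreativeBall_V2) above its base.
; The authority respawns the ball on minigame start, pre-game and option changes, removes it at minigame end, and binds its destruction to an optionally delayed respawn.
; Clients apply material, energy shell and team-based HUD through RepNotify and property-load callbacks, covering a replication race.
; The construction script drives an editor preview, and a function migrates legacy options.
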(event Custom|RespawnBallOnDelay
  (params (DestroyedActor :type Actor%20Object%20Reference))
  (if (and (>= (Variables|UserOptions|GetRespawnDelay) 0.0) (Variables|Default|GetbMinigameInProgress))
    (bind _returnvalue (Utilities|Time|SetTimerbyEvent (delegate Custom|SpawnBallAfterdelay) (Variables|UserOptions|GetRespawnDelay)))
    (Variables|Default|SetRespawnBallTimer _returnvalue)
    (else
      (CallFunction|SpawnBall))))

(event Custom|SpawnBallAfterdelay
  (Utilities|Time|ClearandInvalidateTimerbyHandle (Variables|Default|GetRespawnBallTimer))
  (if (Variables|Default|GetbMinigameInProgress)
    (CallFunction|SpawnBall)))

(event EventBeginPlay
  (bind _fortminigameprogress (Variables|Default|GetFortMinigameProgress))
  (Minigame|AssignMinigameStartedDelegate _fortminigameprogress (EventDispatchers|CreateEvent :!event Custom|OnMinigameStarted)) ; oing it this way instead of using the events because you can't call parent on a component event, but you can on BeginPlay
  (Minigame|BindEventtoMinigameEndedDelegate _fortminigameprogress (EventDispatchers|CreateEvent :!event Custom|OnMinigameEnded))
  (Visibility|SetVisibilityComponents (Variables|Default|GetVisibleInGameComponent) (Utilities|Array|MakeArray (Variables|Default|GetSpawnerBase)) true))

(event Custom|OnMinigameStarted
  (Variables|Default|SetbMinigameInProgress true)
  (Networking|SwitchHasAuthority
    (:Authority
      (CallFunction|SpawnBall))))

(event Custom|OnMinigameEnded
  (Variables|Default|SetbMinigameInProgress false)
  (Networking|SwitchHasAuthority
    (:Authority
      (Utilities|Time|ClearandInvalidateTimerbyHandle (Variables|Default|GetRespawnBallTimer))
      (CallFunction|UnbindandDestroyPreviousBall))))

(event MinigameStateChangedDelegate(FortMinigameProgress)
  (params (Minigame :type Fort%20Minigame%20Object%20Reference) (NewMinigameState :type EFortMinigameState%20Enum :default "PreGame"))
  ;[1 1 1] Respawn the ball each time the game enters pre-game.
  (Networking|SwitchHasAuthority
    (:Authority
      (if (Utilities|Enum|Equal(Enum) NewMinigameState "PreGame")
        (CallFunction|SpawnBall)))))

(event Building|EventOnIsForPreviewingChangedBP
  (params (bNewIsForPreviewing :type Boolean))
  ;[1 1 1] Respawn the ball when any property changes
  (Networking|CanExecuteCosmeticEvents
    (:True
      (CallFunction|UpdateSpawnedBallVisuals) ; After the visuals settings are replicated to the clients\nUpdate the visuals
    )
    (:False
      (if (Building|IsPreviewBuildingActor)
        (CallFunction|UnbindandDestroyPreviousBall)
        (else
          (CallFunction|SpawnBall) ; Respawn Ball with the new options
        )))))

(event OnAnyPropertyChanged(ToyOptionsComponent)
  ;[1 1 1] Respawn the ball when any property changes
  (Networking|CanExecuteCosmeticEvents
    (:True
      (CallFunction|UpdateSpawnedBallVisuals) ; After the visuals settings are replicated to the clients\nUpdate the visuals
    )
    (:False
      (if (Building|IsPreviewBuildingActor)
        (CallFunction|UnbindandDestroyPreviousBall)
        (else
          (CallFunction|SpawnBall) ; Respawn Ball with the new options
        )))))

(event OnPropertiesLoaded(ToyOptionsComponent)
  ;[1 1 1] Respawn the ball when any property changes
  (Networking|CanExecuteCosmeticEvents
    (:True
      (CallFunction|UpdateSpawnedBallVisuals) ; After the visuals settings are replicated to the clients\nUpdate the visuals
    )
    (:False
      (if (Building|IsPreviewBuildingActor)
        (CallFunction|UnbindandDestroyPreviousBall)
        (else
          (CallFunction|SpawnBall) ; Respawn Ball with the new options
        )))))

(event Custom|SetEditorMaterial
  ;[1 1 1] For Setting the Material in editor only
  (bind _loaded_object (Creative|Loading|AsyncLoadEditorSafe (Utilities|Map|Find (Variables|Default|GetMaterialToSet) (Variables|UserOptions|Basic|GetBallMaterial)))
    (:Loaded
      (bind _asmaterial_interface (Utilities|Casting|CastToMaterialInterface _loaded_object))
      (Rendering|Material|SetMaterial (Variables|Default|GetEditorBallMesh) 0 _asmaterial_interface))))

(event EventDestroyed
  (CallFunction|UnbindandDestroyPreviousBall))

(event Custom|EnableHUD
  (params (Instigator :type Actor%20Object%20Reference))
  ;[0.7501 0 1] Receivers
  (Variables|Default|SetShow%20HUD true))

(event Custom|DisableHUD
  (params (Instigator :type Actor%20Object%20Reference))
  ;[0.7501 0 1] Receivers
  (Variables|Default|SetShow%20HUD false))

(fn DisableSimulation
  (Variables|ToyOptions|SetbSimulateBall false)
  (CallFunction|ToggleSimulationEffectsOfBallSpawner false)
  (Networking|SwitchHasAuthority
    (:Authority
      (OptionData|SetPropertyOverrides (Variables|Default|GetToyOptionsComponent) (Utilities|Map|MakeMap "bSimulateBall" "0")))))

(fn GetBoundsComponents
  (params (SelfActor :type Actor%20Object%20Reference))
  (return (Utilities|Array|MakeArray (Variables|Default|GetSpawnerBase))))

(fn GetPropertiesBasedOnBallType
  (params (BallType :type EBallSpawnerPhysTypes%20Enum :default "EBallSpawnerPhysTypes_MAX"))
  (switch Utilities|FlowControl|Switch|SwitchonEBallSpawnerPhysTypes BallType
    (:NewEnumerator0
      ;[1 1 1] Default
      (return 4e+02 0.8 0.4))
    (:NewEnumerator1
      ;[1 1 1] Light
      (return 1e+02 0.9 0.8))
    (:NewEnumerator2
      ;[1 1 1] Heavy
      (return 1e+03 0.8 0.4))
    (:NewEnumerator3
      ;[1 1 1] Slick
      (return 3e+02 0.1 0.05))
    (:NewEnumerator4
      ;[1 1 1] Sporty (Ideal Rocket League Settings)
      (return 2e+02 0.5 0.2))))

(fn MigrateLegacyUserOptions
  (bind _materialindex (Class|GetClassDefaults "/CreativeCoreDevices/Device_BallSpawner_V2.Device_BallSpawner_V2_C"))
  (bind _pawnimpulsemultipler (Class|GetClassDefaults "/CreativeCoreDevices/Device_BallSpawner_V2.Device_BallSpawner_V2_C"))
  (bind _materialindex_2 (Variables|Legacy|GetMaterialIndex))
  (bind _pawnimpulsemultipler_2 (Variables|Legacy|GetPawnImpulseMultipler))
  (Utilities|FlowControl|Sequence
    (:then_0
      (if (!= _pawnimpulsemultipler _pawnimpulsemultipler_2)
        (Variables|UserOptions|Basic|SetPlayerKnockbackAmount (* _pawnimpulsemultipler_2 1e+02))
        (Variables|Legacy|SetPawnImpulseMultipler _pawnimpulsemultipler)
        (LocalVariables|SetMigrated true)))
    (:then_1
      (if (!= _materialindex _materialindex_2)
        (Variables|UserOptions|Basic|SetBallMaterial (Utilities|Select :Index _materialindex_2 :Option%200 "NewEnumerator0" :Option%201 "NewEnumerator1" :Option%202 "NewEnumerator2" :Option%203 "NewEnumerator3" :Option%204 "NewEnumerator4" :Option%205 "NewEnumerator5" :Option%206 "NewEnumerator6"))
        (Variables|Legacy|SetMaterialIndex _materialindex)
        (LocalVariables|SetMigrated true)))
    (:then_2
      ;[1 1 1] Commit the changes so it's saved next time it loads and the logic isn't run again
      (if (LocalVariables|GetMigrated)
        (OptionData|RefreshPropertyOverrides (Variables|Default|GetToyOptionsComponent))))))

(fn OnRep_bSimulateBall
  (Development|SetHiddeninGame (Variables|Default|GetSpawnerBase) false true)
  (CallFunction|ToggleSimulationEffectsOfBallSpawner (Variables|ToyOptions|GetbSimulateBall)))

(fn OnRep_ShowHUD
  (Networking|CanExecuteCosmeticEvents
    (:True
      (bind _spawnedball (Variables|Default|GetSpawnedBall)
        (:then
          (if (Variables|Default|GetShow%20HUD)
            (CallFunction|SpawnOrUpdateHUD)
            (else
              (Class|CreativeBallV2|HideHUD _spawnedball))))
        (:else)))))

(fn OnRep_SpawnedBall
  (Networking|CanExecuteCosmeticEvents
    (:True
      (Variables|Default|GetSpawnedBall
        (:then
          (CallFunction|UpdateSpawnedBallVisuals)
          (Audio|PlaySoundatLocation (Variables|Default|GetStartSimulationSound) (Transformation|GetActorLocation)))
        (:else)))))

(fn SpawnorUpdateHUD
  (Networking|CanExecuteCosmeticEvents
    (:True
      (LocalVariables|SetlocalHostileTeams (Knobs|Team|GetTeamArray self))
      (Utilities|FlowControl|Sequence
        (:then_0
          ;[1 1 1] Parse Team Data
          (if (Creative|TeamOption|IsTeamType (Variables|UserOptions|GetFriendlyTeam) "Neutral")
            (Utilities|Array|Clear (LocalVariables|GetlocalHostileTeams))
            (elif (Creative|TeamOption|IsTeamType (Variables|UserOptions|GetFriendlyTeam) "Any")
              (LocalVariables|SetlocalFriendlyTeam (LocalVariables|GetlocalHostileTeams)) ; Friendly teams take priority over hostile teams.
              (elif (Creative|TeamOption|IsTeamType (Variables|UserOptions|GetFriendlyTeam))
                (Utilities|Array|Add (LocalVariables|GetlocalFriendlyTeam) (- (Creative|TeamOption|GetLegacyTeamIndex (Variables|UserOptions|GetFriendlyTeam) false) 1))))))
        (:then_1
          (if (Variables|UserOptions|GetHUDIcon)
            (bind _spawnedball (Variables|Default|GetSpawnedBall)
              (:then
                (Class|CreativeBallV2|SpawnOrUpdateHUD _spawnedball (Variables|UserOptions|GetHideHUDIconAt) 0 (Variables|UserOptions|GetbRequiresLineOfSight) (LocalVariables|GetlocalFriendlyTeam) (LocalVariables|GetlocalHostileTeams) (Creative|TeamOption|GetLegacyTeamIndex (Variables|UserOptions|GetTeamVisibility)) (Utilities|String|ToString(Text) (Variables|UserOptions|GetFriendlyIconText)) (Utilities|String|ToString(Text) (Variables|UserOptions|GetNeutralIconText)) (Utilities|String|ToString(Text) (Variables|UserOptions|GetHostileIconText)) (Variables|UserOptions|GetHUDTextSize) (Variables|UserOptions|GetbDisplayDistanceText) (Variables|UserOptions|GetbClampToScreen) (Variables|UserOptions|GetbShowOffscreenArrow) (Creative|UI|IconPicker|GetIconLibraryTexture (Variables|UserOptions|GetIconIdentifier))))
              (:else))
            (else
              (bind _spawnedball_2 (Variables|Default|GetSpawnedBall)
                (:then
                  (Class|CreativeBallV2|HideHUD _spawnedball_2))
                (:else)))))))))

(fn SpawnBall
  (bind (_mass _angulardampening _lineardampening) (CallFunction|GetPropertiesBasedonBallType (Variables|UserOptions|Basic|GetType)))
  (CallFunction|UnbindandDestroyPreviousBall)
  (bind _returnvalue (Class|FortPhysicsPawnObject|SpawnPhysicsPawnObject "/CreativeCoreDevices/BallSpawner/CreativeBall_V2.CreativeBall_V2_C" (Math|Transform|MakeTransform (+ (Transformation|GetActorLocation) (Math|Vector|RotateVector (Math|Vector|MakeVector 0.0 0.0 (* (Variables|Default|GetBallRadiusOffsetWhenScaleIsOne) (Variables|UserOptions|Basic|GetSize))) (Transformation|GetActorRotation))) "0, 0, 0" (Math|Vector|MakeVector (Variables|UserOptions|Basic|GetSize) (Variables|UserOptions|Basic|GetSize) (Variables|UserOptions|Basic|GetSize))) (Utilities|Struct|MakePhysicsPawnObjectInitialParameters (Variables|ToyOptions|GetInitialVelocity) _mass _angulardampening _lineardampening (Variables|UserOptions|Basic|GetPlayerForceMultiplier) (/ (Variables|UserOptions|Basic|GetPlayerKnockbackAmount) 1e+02) (Variables|UserOptions|Basic|GetbGravity) (Variables|ToyOptions|GetbEnableGravityOnHit))))
  (bind _output_get (Variables|Default|SetSpawnedBall _returnvalue))
  (Utilities|IsValid _output_get
    (:Is%20Valid
      (Game|BindEventtoOnDestroyed (Variables|Default|GetSpawnedBall) (EventDispatchers|CreateEvent :!event RespawnBallOnDelay))
      (Variables|/CreativeCoreDevices/BallSpawner/CreativeBall_V2.CreativeBall_V2_C|SetKillPawnOnHit (and (Variables|UserOptions|GetbEliminatePlayerWhenTouched) (Variables|Default|GetbMinigameInProgress)) (Variables|Default|GetSpawnedBall)))))

(fn ToggleSimulationEffectsOfBallSpawner
  (params (bSimulationEnabled :type Boolean))
  (bind _spawnerbase (Variables|Default|GetSpawnerBase))
  (bind _spawnerbase_2 (Variables|Default|GetSpawnerBase))
  (if bSimulationEnabled
    (Collision|SetCollisionResponsetoChannel _spawnerbase_2 "ECC_Pawn" "ECR_Ignore")
    (Collision|SetCollisionResponsetoChannel _spawnerbase_2 "ECC_Vehicle" "ECR_Ignore")
    (Networking|CanExecuteCosmeticEvents
      (:True
        (Audio|PlaySoundatLocation (Variables|Default|GetStartSimulationSound) (Transformation|GetActorLocation))))
    (else
      (Actor|DestroyActor (Variables|Default|GetSpawnedBall))
      (Collision|SetCollisionResponsetoChannel _spawnerbase "ECC_Pawn" "ECR_Block")
      (Collision|SetCollisionResponsetoChannel _spawnerbase "ECC_Vehicle" "ECR_Block"))))

(fn UnbindAndDestroyPreviousBall
  (bind _spawnedball (Variables|Default|GetSpawnedBall)
    (:then
      (Game|UnbindEventfromOnDestroyed _spawnedball (EventDispatchers|CreateEvent :!event RespawnBallOnDelay))
      (Actor|DestroyActor _spawnedball))
    (:else)))

(fn UpdateSpawnedBallVisuals
  (Networking|CanExecuteCosmeticEvents
    (:True
      (bind _spawnedball (Variables|Default|GetSpawnedBall)
        (:then
          (CallFunction|SpawnOrUpdateHUD) ; Do this here as well as on the Spawn Ball Repnotify to avoid a race condition on initial spawn where the ball manages to spawn and replicate before the Toy Component Properties are loaded.
          (Class|CreativeBallV2|ChangeBallMaterial _spawnedball (Variables|UserOptions|Basic|GetBallMaterial))
          (Class|CreativeBallV2|ApplyEnergyShell _spawnedball (Variables|UserOptions|GetbEnergyShell)))
        (:else)))))

(fn ConstructionScript
  (bind _editorballmesh (Variables|Default|GetEditorBallMesh))
  (|CallParentFunction)
  (Collision|SetCollisionEnabled (Variables|Default|GetSpawnerBase) "QueryAndPhysics")
  (if (Editor|InEditor)
    (Rendering|SetVisibility (Variables|Default|GetEditorBallMesh) true false)
    (Transformation|SetWorldScale3D (Variables|Default|GetSpawnerBase) (Math|Vector|MakeVector 3.0 3.0 3.0))
    (Transformation|SetWorldScale3D _editorballmesh (Math|Vector|MakeVector (Variables|UserOptions|Basic|GetSize) (Variables|UserOptions|Basic|GetSize) (Variables|UserOptions|Basic|GetSize)))
    (Transformation|SetWorldLocation _editorballmesh (+ (Transformation|GetActorLocation) (Math|Vector|RotateVector (Math|Vector|MakeVector 0.0 0.0 (* (Variables|Default|GetBallRadiusOffsetWhenScaleIsOne) (Variables|UserOptions|Basic|GetSize))) (Transformation|GetActorRotation))) false false)
    (Rendering|SetVisibility (Variables|Default|GetEnergyShell) (Variables|UserOptions|GetbEnergyShell) false)
    (CallFunction|SetEditorMaterial)))
