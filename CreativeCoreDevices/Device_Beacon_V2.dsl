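; Creative Beacon device that marks its location with a team-colored or custom-colored particle beacon, a floating HUD badge, or both.
; BeaconUpdate runs client-side cosmetics, async-loading the particle style and showing or hiding visuals by enabled phase, team, class and show list.
; Option changes and local team or class switches trigger updates.
; Events edit the per-player ShowList, whose OnRep compares a local copy to avoid blinking when other players change.
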(event Custom|Event%20On%20Add%20Player%20To%20Show%20List
  (params (Instigator :type Controller%20Object%20Reference))
  ;[1 0 0.8401] Register/Unregister Player System
  (Default|UpdateShowList true Instigator))

(event Custom|Event%20On%20Remove%20All%20Players%20From%20Show%20List
  (params (Instigator :type Controller%20Object%20Reference))
  ;[1 0 0.8401] Register/Unregister Player System
  (Variables|Transient|SetShowList (Utilities|Array|MakeArray)))

(event Custom|Event%20On%20Remove%20Player%20From%20Show%20List
  (params (Instigator :type Controller%20Object%20Reference))
  ;[1 0 0.8401] Register/Unregister Player System
  (Default|UpdateShowList false Instigator))

(event MinigameEndedDelegate(FortMinigameProgress)
  ;[1 0 0.8401] Register/Unregister Player System
  (Variables|Transient|SetShowList (Utilities|Array|MakeArray)))

(event OnAddedtoMinigame(FortMinigameProgress)
  (params (Minigame :type Fort%20Minigame%20Object%20Reference))
  ;[0 1 0.1154] Force Update when changing class
  (Networking|CanExecuteCosmeticEvents
    (:True
      (Utilities|IsValid Minigame
        (:Is%20Valid
          (Minigame|BindEventtoOnAnyLocalPlayerChangedClass Minigame (EventDispatchers|CreateEvent :!event Beacon%20Update)))))))

(event EventBeginPlay
  ;[0 1 0.1154] Force Updating State when the player changes teams
  (|CallParentFunction)
  (Networking|CanExecuteCosmeticEvents
    (:True
      (bind _asfort_player_controller_athena (Utilities|Casting|CastToFortPlayerControllerAthena (Game|GetPlayerController 0)))
      (FortPlayerControllerAthena|BindEventtoOnLocalPlayerChangedTeams _asfort_player_controller_athena (EventDispatchers|CreateEvent :!event Beacon%20Update)))))

(event OnPropertiesLoaded(ToyOptionsComponent)
  ;[0 1 0.1154] Setting here as parent has no component options (these are defined on the children)
  (CallFunction|UpdateEnabledfromPhase "PreGame")
  (Initialization|Initialize (Variables|Default|GetCreative_UseRestriction_Component) (Variables|Default|GetFortMinigameProgress) 0 0 (Utilities|Struct|MakeCreativeTeamOption "Any") false (Variables|UserOptions|GetFriendly%20Class) (Variables|UserOptions|GetInvertClassSelection) false)
  (CallFunction|BeaconUpdate))

(event OnAnyPropertyChanged(ToyOptionsComponent)
  ;[0 1 0.1154] Setting here as parent has no component options (these are defined on the children)
  (CallFunction|UpdateEnabledfromPhase "PreGame")
  (Initialization|Initialize (Variables|Default|GetCreative_UseRestriction_Component) (Variables|Default|GetFortMinigameProgress) 0 0 (Utilities|Struct|MakeCreativeTeamOption "Any") false (Variables|UserOptions|GetFriendly%20Class) (Variables|UserOptions|GetInvertClassSelection) false)
  (CallFunction|BeaconUpdate))

(event EditorPreview|EventSetupForEditorPreview
  ;[0 1 0.1154] Setting here as parent has no component options (these are defined on the children)
  (CallFunction|UpdateEnabledfromPhase "PreGame")
  (Initialization|Initialize (Variables|Default|GetCreative_UseRestriction_Component) (Variables|Default|GetFortMinigameProgress) 0 0 (Utilities|Struct|MakeCreativeTeamOption "Any") false (Variables|UserOptions|GetFriendly%20Class) (Variables|UserOptions|GetInvertClassSelection) false)
  (CallFunction|BeaconUpdate))

(event EventBeaconUpdate
  (bind _beacon_particle_style (Variables|UserOptions|Basic|GetBeacon%20Particle%20Style))
  ;[1 0.7982 0] Override for Beacon
  (Networking|CanExecuteCosmeticEvents
    (:True
      (if (Utilities|Enum|NotEqual(Enum) "NewEnumerator0" _beacon_particle_style)
        (bind _loaded_object (Creative|Loading|AsyncLoadEditorSafe (Utilities|Struct|BreakFDestructionObjectParticleType (Utilities|Map|Find (Variables|BeaconVisuals|GetBeaconParticleTypes) _beacon_particle_style)))
          (:Loaded
            (bind _asparticle_system (Utilities|Casting|CastToParticleSystem _loaded_object))
            (Variables|Transient|SetBeaconParticlePropTemplate _asparticle_system)
            (Variables|Default|SetParticlePropTemplate 0)))
        (else
          _))
      (if (and (Variables|Default|GetbEnabled) (or (Default|IsLocalPlayerinShowList) (Default|VisibletoPlayerClass)))
        (switch Utilities|FlowControl|Switch|SwitchonEBeaconDisplayStyle (Variables|UserOptions|Basic|GetBeacon%20to%20Show)
          (:NewEnumerator0
            (Class|CreativeValueBarComponent|HideValueBar (Variables|Default|GetFloatingHUDBar)))
          (:NewEnumerator1
            (CallFunction|ShowHUD)
            (CallFunction|HideBeacon))
          (:NewEnumerator2
            (CallFunction|ShowHUD)))
        (if (or (Default|IsLocalPlayerinShowList) (Default|VisibletoPlayerTeam))
          (CallFunction|ShowBeacon)
          (CallFunction|SetBeaconIdentifierParameters)
          (else
            (CallFunction|HideBeacon)))
        (else
          (Class|CreativeValueBarComponent|HideValueBar (Variables|Default|GetFloatingHUDBar))
          (CallFunction|HideBeacon))))))

(fn ActivateFromOtherDevice
  (params (Instigator :type Controller%20Object%20Reference) (Requester :type ERequesterType%20Enum :default "NewEnumerator0"))
  (return false))

(fn BP_GetComponentsBoundingBox
  (params (bNonColliding :type Boolean) (bIncludeFromChildActors :type Boolean))
  (return (CallFunction|GetBoundingBoxwithIgnoredComponents (Utilities|Array|MakeArray (Variables|Default|GetParticlePropPS)) bNonColliding bIncludeFromChildActors)))

(fn CanInteract
  (params (InteractingPawn :type Fort%20Pawn%20Object%20Reference) (InteractionBeingAttempted :type EInteractionBeingAttempted%20Enum :default "FirstInteraction") (InteractionType :type TInteractionType%20Enum :default "IT_NoInteraction"))
  (return false))

(fn GetDeviceColor
  (bind (_value _returnvalue) (Utilities|Map|Find (Variables|Default|GetParticleSystemVectorParameters) "Color"))
  (bind (_value_2 _returnvalue_2) (Utilities|Map|Find (Variables|Default|GetParticleSystemVectorParameters) "Light"))
  (return (select (Utilities|Array|ContainsItem (Variables|/Script/CascadeRuntime.ParticleSystemComponent|GetInstanceParameters (Variables|Default|GetParticlePropPS)) (Utilities|Struct|MakeParticleSysParam "Color" "PSPT_Color")) (Math|Color|SelectColor _value_2 "(R=0.000000,G=0.537697,B=1.000000,A=1.000000)" _returnvalue_2) (Math|Color|SelectColor _value "(R=0.000000,G=0.537697,B=1.000000,A=1.000000)" _returnvalue))))

(fn GetInstigatingController
  (return 0))

(fn GetParticleSystemTemplatetoSpawn
  ;[0.3722 0.1051 1] Return our cached Transient Particle Prop Template
  (return (Variables|Transient|GetBeaconParticlePropTemplate)))

(fn GetColor
  (bind (_base_color _light_color) (Teams|TeamOption|GetTeamColorforTeamOption self (Variables|UserOptions|GetFriendly%20Team)))
  (switch Utilities|FlowControl|Switch|SwitchonECreativeColorSetType (Variables|UserOptions|Basic|GetBeacon%20Color)
    (:NewEnumerator0
      (return "(R=1.000000,G=1.000000,B=1.000000,A=1.000000)"))
    (:NewEnumerator1
      (return (Variables|UserOptions|Basic|GetCustom%20Beacon%20Color)))
    (:NewEnumerator2
      (return (Math|Color|SelectColor (Variables|BeaconVisuals|GetNeutralTeamBeaconColor) _light_color (Creative|TeamOption|IsTeamType (Variables|UserOptions|GetFriendly%20Team) "Neutral"))))
    (:NewEnumerator3
      (bind _valid_controller (CreativeUserRestrictions|ActorHasValidParticipantController (Variables|Default|GetCreative_UseRestriction_Component) (Fortnite|GetFirstLocalFortPlayerController true) true true true))
      (return (Math|Color|SelectColor (Variables|BeaconVisuals|GetNeutralTeamBeaconColor) (Math|Color|SelectColor (Variables|BeaconVisuals|GetFriendlyTeamBeaconColor) (Variables|BeaconVisuals|GetEnemyTeamBeaconColor) (or (Creative|TeamOption|IsTeamType (Variables|UserOptions|GetFriendly%20Team) "Any") (and _valid_controller (Utilities|Enum|Equal(Enum) (Creative|TeamOption|ConverttoTeam (Variables|UserOptions|GetFriendly%20Team)) (Utilities|Enum|BytetoEnumEFortTeam (Team|GetActorTeam (Fortnite|GetFirstLocalFortPlayerController true))))))) (Creative|TeamOption|IsTeamType (Variables|UserOptions|GetFriendly%20Team) "Neutral"))))))

(fn GetSpawnedPickups
  (return 0))

(fn HideBeacon
  (Rendering|SetVisibility (Variables|Default|GetAdditionalVisibleItemRoot) false true)
  (bind _particlepropps (Variables|Default|GetParticlePropPS)
    (:then
      (Components|DestroyComponent _particlepropps))
    (:else)))

(fn IsLocalPlayerInShowList
  (bind _returnvalue (Fortnite|GetFirstLocalFortPlayerController true))
  (Utilities|IsValid _returnvalue
    (:Is%20Valid
      (return (Utilities|Array|ContainsItem (Variables|Transient|GetShowList) (Variables|/Script/Engine.Controller|GetPlayerState _returnvalue))))
    (:Is%20Not%20Valid
      (return false))))

(fn IsEnabled
  (return false))

(fn IsHoldingItemForPawn
  (params (Item :type Item%20Variant%20Handle%20Structure) (Pawn :type Pawn%20Object%20Reference))
  (return false))

(fn MigrateLegacyUserOptions
  (bind (_healthbardistance _showlist _addplayertoshowlist _removeplayerfromshowlist _removeallplayersfromshowlist _localshowlist _enabledindex) (Class|GetClassDefaults "/CreativeCoreDevices/Device_Beacon_V2.Device_Beacon_V2_C"))
  (bind _healthbardistance_2 (Variables|Legacy|GetHealthBarDistance))
  (bind _enabledindex_2 (Variables|UserKnobs|GetEnabledIndex))
  (|CallParentFunction)
  (Utilities|FlowControl|Sequence
    (:then_0
      (Utilities|FlowControl|Sequence
        (:then_0
          (if (!= _healthbardistance_2 _healthbardistance)
            (Variables|UserOptions|SetHide%20HUD%20Icon%20at (/ _healthbardistance_2 1e+02))
            (Variables|SetBy-RefVar _healthbardistance_2 _healthbardistance)
            (LocalVariables|SetMigrateOption true)))
        (:then_1
          (if (!= _enabledindex _enabledindex_2)
            (Variables|UserOptions|Basic|SetEnabledOnPhase (Utilities|Enum|BytetoEnumEEnabledOnPhase (Math|Conversions|ToByte(Integer) (+ _enabledindex_2 1))))
            (Variables|SetBy-RefVar _enabledindex_2 _enabledindex)
            (LocalVariables|SetMigrateOption true)))))
    (:then_1
      (if (LocalVariables|GetMigrateOption)
        (OptionData|RefreshPropertyOverrides (Variables|Default|GetToyOptionsComponent))))))

(fn OnRep_ShowList
  (bind _returnvalue (Fortnite|GetFirstLocalFortPlayerController true))
  (bind _playerstate (Variables|/Script/Engine.Controller|GetPlayerState _returnvalue))
  ;[0 0.6216 1] Only call Beacon Update if either LocalShowList or ShowList array does not contain the Local Player, as this means the Local Player was just added or removed. No need to update the beacon for the local player if another player is added or removed. This prevents random "blinking" of the beacon.
  (Utilities|IsValid _returnvalue
    (:Is%20Valid
      (if (Utilities|Array|ContainsItem (Variables|Transient|GetShowList) _playerstate)
        (if (not (Utilities|Array|ContainsItem (Variables|Transient|GetLocalShowList) _playerstate))
          (CallFunction|BeaconUpdate))
        (Variables|Transient|SetLocalShowList (Variables|Transient|GetShowList))
        (else
          _)))))

(fn SetTriggerCollision
  (params (On :type Boolean))
  (return false))

(fn SetBeaconIdentifierParameters
  (Effects|Components|ParticleSystem|SetEmitterEnable (Variables|Default|GetParticlePropPS) "Letter" (Utilities|Enum|NotEqual(Enum) (Variables|UserOptions|GetBeaconIdentifier) "NewEnumerator0"))
  (Effects|Components|ParticleSystem|SetFloatParameter (Variables|Default|GetParticlePropPS) "U_Index" (Math|Conversions|ToFloat(Integer) (- (Math|Conversions|ToInteger(Byte) (Variables|UserOptions|GetBeaconIdentifier)) 1))))

(fn ShowBeacon
  (bind (_particletype_3_45bbedad49901d6a2ff9fab0205b3f51 _colorparameters_13_112d79bd4c25b6efa3483f992bd13189) (Utilities|Struct|BreakFDestructionObjectParticleType (Utilities|Map|Find (Variables|BeaconVisuals|GetBeaconParticleTypes) (Variables|UserOptions|Basic|GetBeacon%20Particle%20Style))))
  (Rendering|SetVisibility (Variables|Default|GetAdditionalVisibleItemRoot) true true)
  (CallFunction|SpawnParticles)
  (bind _color (CallFunction|GetColor))
  (for _array_element _colorparameters_13_112d79bd4c25b6efa3483f992bd13189
    (Effects|Components|ParticleSystem|SetVectorParameter (Variables|Default|GetParticlePropPS) (Utilities|Struct|BreakFDestructionObjectParticleColorParameter _array_element) (* (Math|Conversions|ToVector(LinearColor) _color) (.ColorParameterMultiplier_7_AE2CF29447670F4CE0F86BA17A6BB6D6 (Utilities|Struct|BreakFDestructionObjectParticleColorParameter _array_element))))))

(fn ShowHUD
  (if (and (not (Editor|InEditor)) (or (Default|VisibletoPlayerTeam) (Default|IsLocalPlayerinShowList)))
    (LocalVariables|SetlocalHostileTeams (Knobs|Team|GetTeamArray self))
    (Utilities|FlowControl|Sequence
      (:then_0
        ;[1 1 1] Parse Team Data
        (if (Creative|TeamOption|IsTeamType (Variables|UserOptions|GetFriendly%20Team) "Neutral")
          (Utilities|Array|Clear (LocalVariables|GetlocalHostileTeams))
          (else
            (bind _index (CreativeTeam|GetTeamPVPIndex (Variables|UserOptions|GetFriendly%20Team))
              (:Any%20/%20All
                (LocalVariables|SetlocalFriendlyTeam (LocalVariables|GetlocalHostileTeams)) ; Friendly teams take priority over hostile teams.
              )
              (:Team%20Index
                (Utilities|Array|Add (LocalVariables|GetlocalFriendlyTeam) (Math|Conversions|ToInteger(Byte) _index)))))))
      (:then_1
        (Class|CreativeValueBarComponent|SpawnOrUpdateValueBar (Variables|Default|GetFloatingHUDBar) (Variables|Default|GetSelectionMesh) self "/CreativeCoreDevices/Beacon/Creative_FloatingValueBar_MVVM_Container.Creative_FloatingValueBar_MVVM_Container_C" (Variables|UserOptions|GetCustomWidget) (* (Variables|UserOptions|GetHide%20HUD%20Icon%20at) 1e+02) 0 -1.0 8e+01 (Variables|UserOptions|GetRequires%20Line%20of%20Sight) (LocalVariables|GetlocalFriendlyTeam) (LocalVariables|GetlocalHostileTeams) (Creative|TeamOption|GetLegacyTeamIndex (Variables|UserOptions|GetTeam%20Visibility)) false (Utilities|String|ToString(Text) (Variables|UserOptions|GetFriendly%20Icon%20Text)) (Utilities|String|ToString(Text) (Variables|UserOptions|GetNeutral%20Icon%20Text)) (Utilities|String|ToString(Text) (Variables|UserOptions|GetHostile%20Icon%20Text)) (Variables|UserOptions|GetHUD%20Text%20Size) false (Variables|UserOptions|GetDisplayDistanceText) (Variables|UserOptions|GetClampToScreen) (Variables|UserOptions|GetShowOffscreenArrow) (Creative|UI|IconPicker|GetIconLibraryTexture (Variables|UserOptions|GetIcon%20Identifier)) (Utilities|Enum|Equal(Enum) (Variables|UserOptions|GetBadge%20UI%20Style) "NewEnumerator0") "(X=1.000000,Y=1.000000)" false (Default|IsLocalPlayerinShowList))
        (Class|CreativeValueBarComponent|SetCustomFontStyle (Variables|Default|GetFloatingHUDBar) (Variables|UserOptions|GetText%20Font%20Style))
        (if (Variables|UserOptions|Basic|GetBadge%20uses%20Beacon%20Color)
          (bind _color (CallFunction|GetColor))
          (Class|CreativeValueBarComponent|SetCustomColor (Variables|Default|GetFloatingHUDBar) true _color :!class /Game/Creative/Devices/Common/Components/Creative_ValueBar_Component.Creative_ValueBar_Component_C)
          (else
            (Class|CreativeValueBarComponent|SetCustomColor :self (Variables|Default|GetFloatingHUDBar) :UseCustomColor false :!class /Game/Creative/Devices/Common/Components/Creative_ValueBar_Component.Creative_ValueBar_Component_C)))))
    (else
      (Class|CreativeValueBarComponent|HideValueBar (Variables|Default|GetFloatingHUDBar)))))

(fn UpdateEnabledFromPhase
  (params (local%20Phase :type EFortMinigameState%20Enum :default "PreGame"))
  (bind _local_enabled (LocalVariables|Getlocal%20Enabled))
  (bind _benabled (Variables|Default|GetbEnabled))
  (bind _returnvalue (Math|Boolean|NotEqual(Boolean) _benabled _local_enabled))
  ;[1 1 1] Overriding parent since we're using a different variable for now.
  (LocalVariables|Setlocal%20Enabled (Utilities|Select :Index (Variables|UserOptions|Basic|GetEnabledOnPhase) :NewEnumerator1 false :NewEnumerator0 true :NewEnumerator2 (Utilities|Enum|Equal(Enum) local%20Phase "PreGame") :NewEnumerator4 (Utilities|Array|ContainsItem (Variables|Default|GetGameplay%20Phases) local%20Phase) :NewEnumerator3 false))
  (Networking|SwitchHasAuthority
    (:Authority
      (if _returnvalue
        (Variables|Default|SetbEnabled (LocalVariables|Getlocal%20Enabled))))
    (:Remote
      (if (not _returnvalue)
        (CallFunction|BeaconUpdate)))))

(fn UpdateShowList
  (params (Add :type Boolean) (Instigator :type Controller%20Object%20Reference))
  (bind _showlist (Variables|Transient|GetShowList))
  (Utilities|IsValid Instigator
    (:Is%20Valid
      (bind _output_get (LocalVariables|SetLocal%20PlayerState (Variables|/Script/Engine.Controller|GetPlayerState Instigator)))
      (Utilities|IsValid _output_get
        (:Is%20Valid
          (Utilities|Set|AddItems (LocalVariables|GetLocal%20Instigator%20Cache) _showlist)
          (LocalVariables|SetHadInstigators (Math|Conversions|ToBoolean(Integer) (Utilities|Array|Length _showlist)))
          (if (Variables|GetAdd)
            (if (not (Utilities|Set|ContainsItem (LocalVariables|GetLocal%20Instigator%20Cache) (LocalVariables|GetLocal%20PlayerState)))
              (Utilities|Set|Add (LocalVariables|GetLocal%20Instigator%20Cache) (LocalVariables|GetLocal%20PlayerState)))
            (else
              (bind _returnvalue (Utilities|Set|Remove (LocalVariables|GetLocal%20Instigator%20Cache) (LocalVariables|GetLocal%20PlayerState)))
              (if _returnvalue
                _)))
          (bind _result (Utilities|Set|ToArray (LocalVariables|GetLocal%20Instigator%20Cache)))
          (Variables|Transient|SetShowList _result))))))

(fn ConstructionScript
  (|CallParentFunction))

(fn VisibleToPlayerClass
  (bind _isvalidcontroller (LocalVariables|GetisValidController))
  (bind _valid_controller (CreativeUserRestrictions|ActorHasValidParticipantController (Variables|Default|GetCreative_UseRestriction_Component) (Fortnite|GetFirstLocalFortPlayerController true) true true true))
  (LocalVariables|SetisValidController _valid_controller)
  (bind _currentminigame (Variables|/Script/FortniteGame.FortMinigameLogicComponent|GetCurrentMinigame (Variables|Default|GetFortMinigameProgress))
    (:then
      (return (or (Minigame|IsPregame _currentminigame) _isvalidcontroller)))
    (:else
      (return (Math|Boolean|ORBoolean _isvalidcontroller (Creative|ClassOption|IsClassType (Variables|UserOptions|GetFriendly%20Class) "Any") (Editor|InEditor))))))

(fn VisibleToPlayerTeam
  (bind (_friendly_team_teamtype _friendly_team_teamindex) (Utilities|Struct|BreakCreativeTeamOption (Variables|GetFriendly%20Team_TeamType)))
  (bind (_team_visibility_teamtype _team_visibility_teamindex) (Utilities|Struct|BreakCreativeTeamOption (Variables|GetTeam%20Visibility_TeamType)))
  (LocalVariables|SetLocalPawnTeamIndex (Team|GetActorTeam (Fortnite|GetFirstLocalFortPlayerController true)))
  (return (Math|Boolean|ORBoolean (Creative|TeamOption|IsTeamType (Variables|UserOptions|GetTeam%20Visibility) "Any") (Math|Boolean|ANDBoolean (not (Creative|TeamOption|IsTeamType (Variables|UserOptions|GetTeam%20Visibility))) (Utilities|Enum|Equal(Enum) _team_visibility_teamtype _friendly_team_teamtype) (== _team_visibility_teamindex _friendly_team_teamindex)) (and (Creative|TeamOption|IsTeamType (Variables|UserOptions|GetTeam%20Visibility) "Friendlies") (or (and (Creative|TeamOption|IsTeamType (Variables|UserOptions|GetFriendly%20Team)) (== (Creative|TeamOption|ConverttoTeam (Variables|UserOptions|GetFriendly%20Team)) (LocalVariables|GetLocalPawnTeamIndex))) (Creative|TeamOption|IsTeamType (Variables|UserOptions|GetFriendly%20Team) "Any"))) (and (Creative|TeamOption|IsTeamType (Variables|UserOptions|GetTeam%20Visibility)) (== (Creative|TeamOption|ConverttoTeam (Variables|UserOptions|GetTeam%20Visibility)) (LocalVariables|GetLocalPawnTeamIndex))) (Editor|InEditor))))
