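; Creative billboard device that displays configurable or event-driven text on a widget with an optional border mesh.
; Option changes and world setup async-load the font, update the border and push details to the widget, adding a backface widget for two-sided display.
; An enabled component and minigame play mode drive visibility and collision.
; Replicated Verse overrides for text, size and border reset when play mode ends; legacy font settings are migrated.
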(event EventOnWorldReady
  ;[0 1 0.2048 0.1] Initialization
  (|CallParentFunction)
  (Utilities|FlowControl|Sequence
    (:then_0
      (CallFunction|CheckIsFontLoadRequired)
      (CallFunction|UpdateBorderMesh)
      (Networking|CanExecuteCosmeticEvents
        (:True
          (Default|SetViewDistance)
          (Class|BillboardWidgetComp|SetResolutionMultiplier (Variables|Default|GetBillboard_WidgetComp) (Variables|UserOptions|GetResolution%20Multiplier) :!class /CreativeCoreDevices/Billboard/Billboard_WidgetComp.Billboard_WidgetComp_C)
          (Default|UpdateEditorHiddeninGame)
          (CallFunction|UpdateBillboard)
          (bind _billboard_widgetcomp_backface (Variables|Default|GetBillboard_WidgetComp_Backface)
            (:then
              (Class|BillboardWidgetComp|SetResolutionMultiplier _billboard_widgetcomp_backface (Variables|UserOptions|GetResolution%20Multiplier) :!class /CreativeCoreDevices/Billboard/Billboard_WidgetComp.Billboard_WidgetComp_C))
            (:else)))))
    (:then_1
      (Class|CreativeEnabledComponent|Initialize (Variables|Default|GetCreative_Enabled_Component) (Variables|Default|GetFortMinigameLogic) (Variables|UserOptions|GetEnabledDuringPhase)) ; Has to initialize after the creation of billboard
      ;;[1 1 1] Update widget visibility and material if the Display Mode is changed, otherwise it won't refresh in Creative / VK-Edit on option changed
      (if (Utilities|Enum|NotEqual(Enum) (Variables|UserOptions|GetDisplayMode) (Variables|Default|GetCachedDisplayMode))
        (Variables|Default|SetCachedDisplayMode (Variables|UserOptions|GetDisplayMode))
        (CallFunction|UpdateWidgetVisibility)))))

(event EditorPreview|EventSetupForEditorPreview
  ;[0 1 0.2048 0.1] Initialization
  (|CallParentFunction)
  (Utilities|FlowControl|Sequence
    (:then_0
      (CallFunction|CheckIsFontLoadRequired)
      (CallFunction|UpdateBorderMesh)
      (Networking|CanExecuteCosmeticEvents
        (:True
          (Default|SetViewDistance)
          (Class|BillboardWidgetComp|SetResolutionMultiplier (Variables|Default|GetBillboard_WidgetComp) (Variables|UserOptions|GetResolution%20Multiplier) :!class /CreativeCoreDevices/Billboard/Billboard_WidgetComp.Billboard_WidgetComp_C)
          (Default|UpdateEditorHiddeninGame)
          (CallFunction|UpdateBillboard)
          (bind _billboard_widgetcomp_backface (Variables|Default|GetBillboard_WidgetComp_Backface)
            (:then
              (Class|BillboardWidgetComp|SetResolutionMultiplier _billboard_widgetcomp_backface (Variables|UserOptions|GetResolution%20Multiplier) :!class /CreativeCoreDevices/Billboard/Billboard_WidgetComp.Billboard_WidgetComp_C))
            (:else)))))
    (:then_1
      (Class|CreativeEnabledComponent|Initialize (Variables|Default|GetCreative_Enabled_Component) (Variables|Default|GetFortMinigameLogic) (Variables|UserOptions|GetEnabledDuringPhase)) ; Has to initialize after the creation of billboard
      ;;[1 1 1] Update widget visibility and material if the Display Mode is changed, otherwise it won't refresh in Creative / VK-Edit on option changed
      (if (Utilities|Enum|NotEqual(Enum) (Variables|UserOptions|GetDisplayMode) (Variables|Default|GetCachedDisplayMode))
        (Variables|Default|SetCachedDisplayMode (Variables|UserOptions|GetDisplayMode))
        (CallFunction|UpdateWidgetVisibility)))))

(event OnPropertiesLoaded(ToyOptionsComponent)
  ;[0 1 0.2048 0.1] Initialization
  (Variables|Default|SetCachedDisplayMode (Variables|UserOptions|GetDisplayMode))
  (Utilities|FlowControl|Sequence
    (:then_0
      (CallFunction|CheckIsFontLoadRequired)
      (CallFunction|UpdateBorderMesh)
      (Networking|CanExecuteCosmeticEvents
        (:True
          (Default|SetViewDistance)
          (Class|BillboardWidgetComp|SetResolutionMultiplier (Variables|Default|GetBillboard_WidgetComp) (Variables|UserOptions|GetResolution%20Multiplier) :!class /CreativeCoreDevices/Billboard/Billboard_WidgetComp.Billboard_WidgetComp_C)
          (Default|UpdateEditorHiddeninGame)
          (CallFunction|UpdateBillboard)
          (bind _billboard_widgetcomp_backface (Variables|Default|GetBillboard_WidgetComp_Backface)
            (:then
              (Class|BillboardWidgetComp|SetResolutionMultiplier _billboard_widgetcomp_backface (Variables|UserOptions|GetResolution%20Multiplier) :!class /CreativeCoreDevices/Billboard/Billboard_WidgetComp.Billboard_WidgetComp_C))
            (:else)))))
    (:then_1
      (Class|CreativeEnabledComponent|Initialize (Variables|Default|GetCreative_Enabled_Component) (Variables|Default|GetFortMinigameLogic) (Variables|UserOptions|GetEnabledDuringPhase)) ; Has to initialize after the creation of billboard
      ;;[1 1 1] Update widget visibility and material if the Display Mode is changed, otherwise it won't refresh in Creative / VK-Edit on option changed
      (if (Utilities|Enum|NotEqual(Enum) (Variables|UserOptions|GetDisplayMode) (Variables|Default|GetCachedDisplayMode))
        (Variables|Default|SetCachedDisplayMode (Variables|UserOptions|GetDisplayMode))
        (CallFunction|UpdateWidgetVisibility)))))

(event OnAnyPropertyChanged(ToyOptionsComponent)
  ;[0 1 0.2048 0.1] Initialization
  (Utilities|FlowControl|Sequence
    (:then_0
      (CallFunction|CheckIsFontLoadRequired)
      (CallFunction|UpdateBorderMesh)
      (Networking|CanExecuteCosmeticEvents
        (:True
          (Default|SetViewDistance)
          (Class|BillboardWidgetComp|SetResolutionMultiplier (Variables|Default|GetBillboard_WidgetComp) (Variables|UserOptions|GetResolution%20Multiplier) :!class /CreativeCoreDevices/Billboard/Billboard_WidgetComp.Billboard_WidgetComp_C)
          (Default|UpdateEditorHiddeninGame)
          (CallFunction|UpdateBillboard)
          (bind _billboard_widgetcomp_backface (Variables|Default|GetBillboard_WidgetComp_Backface)
            (:then
              (Class|BillboardWidgetComp|SetResolutionMultiplier _billboard_widgetcomp_backface (Variables|UserOptions|GetResolution%20Multiplier) :!class /CreativeCoreDevices/Billboard/Billboard_WidgetComp.Billboard_WidgetComp_C))
            (:else)))))
    (:then_1
      (Class|CreativeEnabledComponent|Initialize (Variables|Default|GetCreative_Enabled_Component) (Variables|Default|GetFortMinigameLogic) (Variables|UserOptions|GetEnabledDuringPhase)) ; Has to initialize after the creation of billboard
      ;;[1 1 1] Update widget visibility and material if the Display Mode is changed, otherwise it won't refresh in Creative / VK-Edit on option changed
      (if (Utilities|Enum|NotEqual(Enum) (Variables|UserOptions|GetDisplayMode) (Variables|Default|GetCachedDisplayMode))
        (Variables|Default|SetCachedDisplayMode (Variables|UserOptions|GetDisplayMode))
        (CallFunction|UpdateWidgetVisibility)))))

(event Custom|OnFontAsyncLoadCompleted
  ;[0 1 0.2048 0.1] Initialization
  (Utilities|FlowControl|Sequence
    (:then_0
      (CallFunction|CheckIsFontLoadRequired)
      (CallFunction|UpdateBorderMesh)
      (Networking|CanExecuteCosmeticEvents
        (:True
          (Default|SetViewDistance)
          (Class|BillboardWidgetComp|SetResolutionMultiplier (Variables|Default|GetBillboard_WidgetComp) (Variables|UserOptions|GetResolution%20Multiplier) :!class /CreativeCoreDevices/Billboard/Billboard_WidgetComp.Billboard_WidgetComp_C)
          (Default|UpdateEditorHiddeninGame)
          (CallFunction|UpdateBillboard)
          (bind _billboard_widgetcomp_backface (Variables|Default|GetBillboard_WidgetComp_Backface)
            (:then
              (Class|BillboardWidgetComp|SetResolutionMultiplier _billboard_widgetcomp_backface (Variables|UserOptions|GetResolution%20Multiplier) :!class /CreativeCoreDevices/Billboard/Billboard_WidgetComp.Billboard_WidgetComp_C))
            (:else)))))
    (:then_1
      (Class|CreativeEnabledComponent|Initialize (Variables|Default|GetCreative_Enabled_Component) (Variables|Default|GetFortMinigameLogic) (Variables|UserOptions|GetEnabledDuringPhase)) ; Has to initialize after the creation of billboard
      ;;[1 1 1] Update widget visibility and material if the Display Mode is changed, otherwise it won't refresh in Creative / VK-Edit on option changed
      (if (Utilities|Enum|NotEqual(Enum) (Variables|UserOptions|GetDisplayMode) (Variables|Default|GetCachedDisplayMode))
        (Variables|Default|SetCachedDisplayMode (Variables|UserOptions|GetDisplayMode))
        (CallFunction|UpdateWidgetVisibility)))))

(event OnAddedtoMinigame(FortMinigameLogic)
  (params (Minigame :type Fort%20Minigame%20Object%20Reference))
  ;[0 1 0.2048 0.1] Initialization
  (Utilities|IsValid Minigame
    (:Is%20Valid
      (Minigame|BindEventtoOnPlayModeChangedDelegate Minigame (EventDispatchers|CreateEvent :!event Custom|Play%20Mode%20Changed))
      (CallFunction|PlayModeChanged Minigame (Minigame|IsinPlayMode Minigame :!class /Script/FortniteGame.FortMinigame)))))

(event OnEnabledStateChanged(Creative_Enabled_Component)
  (params (Enabled :type Boolean))
  ;[0 0.2878 1 0.1] State Changes
  (CallFunction|UpdateWidgetVisibility)
  (CallFunction|UpdateBorderMesh))

(event Custom|ReceiveShowTextEvent
  (params (Instigator :type Controller%20Object%20Reference))
  ;[0.1652 0 1 0.1] Gameplay Events
  (Class|CreativeEnabledComponent|SetEnabled (Variables|Default|GetCreative_Enabled_Component) true))

(event Custom|Play%20Mode%20Changed
  (params (Minigame :type Fort%20Minigame%20Object%20Reference) (bIsInPlayMode :type Boolean))
  ;[0 0.2878 1 0.1] State Changes
  (Variables|Default|SetIsInPlayMode bIsInPlayMode)
  (CallFunction|UpdateInteractionCollision bIsInPlayMode)
  (CallFunction|UpdateWidgetVisibility)
  (CallFunction|UpdateBorderMesh)
  (if (not bIsInPlayMode)
    (Verse|ResetVerseOverrides)))

(event Custom|ReceiveHideTextEvent
  (params (Instigator :type Controller%20Object%20Reference))
  ;[0.1652 0 1 0.1] Gameplay Events
  (Class|CreativeEnabledComponent|SetEnabled (Variables|Default|GetCreative_Enabled_Component) false))

(event Custom|Update%20Display%20Event
  (params (Instigator :type Controller%20Object%20Reference) (Type :type EPayload_DataType%20Enum :default "NewEnumerator0") (Text :type Text) (Number :type Float%20%28double-precision%29))
  ;[0.1652 0 1 0.1] Gameplay Events
  ;;[1 0.759 0.6475] Update String Payload from Remote Source
  (if (Utilities|Enum|NotEqual(Enum) Type "NewEnumerator0")
    (Variables|Verse|SetText%20From%20Event (select (Utilities|Enum|Equal(Enum) Type "NewEnumerator1") Text (Utilities|Text|ToText(Integer64) (Math|Float|TruncatetoInteger64 Number))))
    (CallFunction|UpdateBillboard)))

(event OnRemovedfromMinigame(FortMinigameLogic)
  (params (Minigame :type Fort%20Minigame%20Object%20Reference))
  ;[0.1719 0.005371 0.005371 0.328] Minigame Removal - Unbind from Play Mode Delegate.
  (Utilities|IsValid Minigame
    (:Is%20Valid
      (Minigame|UnbindEventfromOnPlayModeChangedDelegate Minigame (EventDispatchers|CreateEvent :!event Custom|Play%20Mode%20Changed)))))

(event Custom|DisplayLocalizedMessage
  ;[0 0.0615 1] LocalizableMessage
  (CallFunction|UpdateBillboard))

(event EventBeginPlay
  ;[0 0.0615 1] LocalizableMessage
  (|CallParentFunction)
  (Networking|CanExecuteCosmeticEvents
    (:True
      (EventDispatchers|BindEventtoOnMessageLocalized (Variables|Default|GetFortLocalizableMessage) (delegate Custom|DisplayLocalizedMessage)))))

(event EventPostFinishSpawning
  ;[1 1 1] Post Finish Spawning
  (|CallParentFunction)
  (CallFunction|CheckIsFontLoadRequired)
  (bind _true (CallFunction|MeshIsVisible))
  (Collision|SetCollisionEnabled (Variables|Debug|GetStaticMeshComponent) (select _true "QueryAndPhysics" "NoCollision")))

(event Custom|CheckIsFontLoadRequired
  (bind _font (Variables|UserOptions|Basic|GetFont))
  (bind _fontobject (Utilities|Struct|BreakFortFontOptionData _font))
  ;[0.15 0.055 0.1215 0.5] Font Loading -- the font needs to be resident before it can be used in the widget
  ;; Early out cases: we're loading a font, or the font is already loaded and matches what is configured
  (if (Variables|Default|GetFontIsLoading)
    _
    (elif (and (Utilities|IsValid (Variables|Default|GetLoadedFont) :!class /Script/Engine.KismetSystemLibrary) (== (Utilities|ResolveSoftReference (Variables|Default|GetLoadedFont)) _fontobject))
      _
      (else
        ;; Clear the existing font reference and set loading flag
        (Utilities|FlowControl|Sequence
          (:then_0
            (Variables|Default|SetFontIsLoading true)
            (Variables|Default|SetLoadedFont 0))
          (:then_1
            (bind _object (Utilities|AsyncLoadAsset _fontobject)
              (:Completed
                ;; Clear the loading flag and trigger the completion event
                (Variables|Default|SetFontIsLoading false)
                (bind _asfont (Utilities|Casting|CastToFont _object))
                (Variables|Default|SetLoadedFont _asfont)
                (CallFunction|OnFontAsyncLoadCompleted)))))))))

(event Custom|Update%20Border%20Mesh
  (bind _fortminigamelogic (Variables|Default|GetFortMinigameLogic))
  (bind _play_mode (Class|CreativeCommonDeviceFunctionLibrary|IsLogicinPlayMode _fortminigamelogic))
  (bind _staticmeshcomponent (Variables|Debug|GetStaticMeshComponent))
  ;[1 0.803 0 0.044] Updating visibility of Mesh
  (bind _true (CallFunction|MeshIsVisible))
  (Collision|SetCollisionEnabled _staticmeshcomponent (select _true "QueryAndPhysics" "NoCollision"))
  (Rendering|SetVisibility _staticmeshcomponent (or _true (not _play_mode)) false)
  (bind _loaded_object (Creative|Loading|AsyncLoadEditorSafe (select (or _true _play_mode) (Variables|Default|GetDefaultMaterial) (Variables|Default|GetClearMaterial)))
    (:Loaded
      (bind _asmaterial_interface (Utilities|Casting|CastToMaterialInterface _loaded_object))
      (bind _output_get (Variables|Default|SetCachedMaterial _asmaterial_interface))
      (Rendering|Material|SetMaterial (Variables|Debug|GetStaticMeshComponent) 0 _output_get)
      (bind _billboard_widgetcomp (Variables|Default|GetBillboard_WidgetComp)
        (:then
          (bind _true_2 (CallFunction|MeshIsVisible))
          (Transformation|SetRelativeRotation _billboard_widgetcomp (select _true_2 (Variables|Default|GetBackgroundRelativeRotationWithBorder) (Variables|Default|GetBackgroundRelativeRotationWithoutBorder)) false false))
        (:else)))))

(fn CheckTwoSidedConfig
  (bind _showborder (Variables|UserOptions|GetShowBorder))
  (bind _returnvalue (not _showborder))
  (bind _displaymode (Variables|UserOptions|GetDisplayMode))
  (bind _returnvalue_2 (Utilities|Enum|Equal(Enum) _displaymode "NewEnumerator1"))
  (bind _returnvalue_3 (and _returnvalue _returnvalue_2))
  ;[1 1 1] FIX FNDOC-2493 | Show Double Sided Widget for the Gameplay Only option.
  (if (Utilities|Enum|Equal(Enum) (Variables|UserOptions|GetEnabledDuringPhase) "NewEnumerator4")
    (return _returnvalue_3 (Variables|Default|GetIsInPlayMode))
    (else
      (return _returnvalue_3 _returnvalue_3))))

(fn CreateOrUpdateBackfaceWidget
  (bind (_istwosided _isusingtwosidedmaterial) (CallFunction|CheckTwoSidedConfig))
  (if _istwosided
    (if (not _isusingtwosidedmaterial)
      (Variables|Default|GetBillboard_WidgetComp_Backface
        (:then
          _)
        (:else
          (bind _returnvalue (AddComponent|NODE_AddBillboard_WidgetComp-0 :bManualAttachment true))
          (bind _output_get (Variables|Default|SetBillboard_WidgetComp_Backface _returnvalue))
          (Animation|SetTickwhenOffscreen _output_get true)
          (Transformation|AttachComponentToComponent (Variables|Default|GetBillboard_WidgetComp_Backface) (Variables|Default|GetWidgetAnchor_Backface) "None" "KeepRelative" "KeepRelative" "KeepRelative" true)))
      (return)
      (else
        _))
    (else
      _))
  (bind _billboard_widgetcomp_backface (Variables|Default|GetBillboard_WidgetComp_Backface)
    (:then
      (Components|Activation|Deactivate _billboard_widgetcomp_backface)
      (Components|DestroyComponent _billboard_widgetcomp_backface))
    (:else
      _))
  (return))

(fn GetBillboardDetails
  (bind (_fontobject _fonttypeface) (Utilities|Struct|BreakFortFontOptionData (Variables|UserOptions|Basic|GetFont)))
  (bind _sizevalue (Verse|GetTextSize))
  (return (GetDetails|GetBillboardText) (Variables|UserOptions|Basic|GetTextColor) (Variables|UserOptions|Basic|GetBackgroundColor) (Variables|UserOptions|Basic|GetTextJustification) _sizevalue (GetDetails|GetOutlineStrength) (Utilities|ResolveSoftReference _fontobject) _fonttypeface (Variables|UserOptions|Basic|GetShadow) (/ (Math|Conversions|ToFloat(Integer) (Variables|UserOptions|Basic|GetTextSize)) (Variables|Default|GetShadow%20Divisor))))

(fn GetBillboardText
  (return (select (Utilities|Text|TextIsEmpty (Variables|Verse|GetText%20From%20Event)) (Variables|UserOptions|Basic|GetText) (Variables|Verse|GetText%20From%20Event))))

(fn GetCollisionStaticMesh
  (return "0, 0, 0" (Variables|/Script/Engine.StaticMeshComponent|GetStaticMesh (Variables|Debug|GetStaticMeshComponent))))

(fn GetCreativeActorBounds
  (bind (_origin _boxextent) (Collision|GetComponentBounds (Variables|Default|GetCreativePreviewOnlyMesh)))
  ;[1 1 1] Due to the root mesh having the pivot at the bottom use the CreativePreviewOnlyMesh instead.
  (return true _boxextent))

(fn GetCreativeActorOrigin
  ;[1 1 1] Override the origin with the widget location due to the root mesh having the pivot at the bottom.
  (return true (Variables|/Script/Engine.SceneComponent|GetRelativeLocation (Variables|Default|GetBillboard_WidgetComp))))

(fn GetCreativeActorRelativeTransform
  (return false 0))

(fn GetGridAxisGuideAttachmentPoint
  (return false false "0, 0, 0"))

(fn GetMeshComponents
  (return (Utilities|Array|MakeArray (Variables|Debug|GetStaticMeshComponent))))

(fn GetOutlineStrength
  (return (Math|Float|Truncate (* (Math|Conversions|ToInteger(Byte) (Variables|UserOptions|Basic|GetOutline)) (/ (Math|Conversions|ToFloat(Integer) (Variables|UserOptions|Basic|GetTextSize)) (Variables|Default|GetMinTextSize))))))

(fn GetOverrideMeshMaterial
  (params (MeshComp :type Static%20Mesh%20Component%20Object%20Reference) (MatIdx :type Integer))
  (if (Utilities|Equal(Object) MeshComp (Variables|Debug|GetStaticMeshComponent))
    (return (Variables|Default|GetCachedMaterial))
    (else
      (return 0))))

(fn GetShowBorder
  (return (Utilities|Select :Index (Variables|Verse|GetShow%20Border%20Override) :NewEnumerator0 (Variables|UserOptions|GetShowBorder) :NewEnumerator1 true :NewEnumerator2 false)))

(fn GetText
  (return (Utilities|String|ToString(Text) (select (Utilities|Text|TextIsEmpty (Variables|Verse|GetText%20From%20Event)) (Variables|UserOptions|Basic|GetText) (Variables|Verse|GetText%20From%20Event)))))

(fn GetTextSize
  (return (Math|Integer|Clamp(Integer) (Math|Integer|SelectInt (Variables|UserOptions|Basic|GetTextSize) (Variables|Verse|GetText%20Size%20Override) (== (Variables|Verse|GetText%20Size%20Override) -1)) 8 24)))

(fn IgnoreWorldGridScaleOverride
  (return false))

(fn IsTargetableByLinearMover
  (return true))

(fn MeshIsVisible
  (bind _isshown (Verse|GetShowBorder))
  (return (and _isshown (Class|CreativeEnabledComponent|IsEnabled (Variables|Default|GetCreative_Enabled_Component)))))

(fn MigrateLegacyUserOptions
  (bind (_text _font _showborder _backgroundcolor _textjustification _textsize _displaymode _viewdistance _textcolor _enabledduringphase _textfont) (Class|GetClassDefaults "/CreativeCoreDevices/Device_Billboard_V2.Device_Billboard_V2_C"))
  (|CallParentFunction)
  (Networking|SwitchHasAuthority
    (:Authority
      (Utilities|FlowControl|Sequence
        (:then_0
          ; Migrate legacy font setting
          (if (Utilities|Enum|NotEqual(Enum) (Variables|UserOptions|Legacy|GetTextFont) _textfont)
            (Variables|UserOptions|Basic|SetFont (Utilities|Struct|MakeFortFontOptionData (Utilities|Select :Index (Variables|UserOptions|Legacy|GetTextFont) :NewEnumerator0 "/Engine/EngineFonts/Roboto.Roboto" :NewEnumerator1 "/FortUILibrary/Fonts/Burbank.Burbank" :NewEnumerator2 "/FortUILibrary/Fonts/NotoSans.NotoSans") (Utilities|Select :Index (Variables|UserOptions|Legacy|GetTextFont) :NewEnumerator0 "Bold" :NewEnumerator1 "Medium" :NewEnumerator2 "NotoSans-Regular")))
            (Variables|UserOptions|Legacy|SetTextFont _textfont)
            (LocalVariables|SetOptionMigrated true)))
        (:then_1
          ; save this instance's uproperties to save data if we migrated anything
          (if (LocalVariables|GetOptionMigrated)
            (OptionData|RefreshPropertyOverrides (Variables|Default|GetToyOptionsComponent))))))))

(fn OnRep_ShowBorderOverride
  (CallFunction|UpdateBorderMesh)
  (CallFunction|UpdateBillboard))

(fn OnRep_TextSizeOverride
  (CallFunction|UpdateBillboard))

(fn OverrideText
  (params (Text :type Localizable%20Message%20Structure))
  (Networking|SwitchHasAuthority
    (:Authority
      (Class|FortLocalizableMessageComponent|LocalizeandSetMessage (Variables|Default|GetFortLocalizableMessage) Text "Text From Event" "Text"))))

(fn OverrideTextSize
  (params (InTextSize :type Integer))
  (Networking|SwitchHasAuthority
    (:Authority
      (Variables|Verse|SetText%20Size%20Override (Math|Integer|Clamp(Integer) InTextSize 8 24)))))

(fn ResetVerseOverrides
  (Class|FortLocalizableMessageComponent|LocalizeandSetMessage (Variables|Default|GetFortLocalizableMessage) (Utilities|Struct|MakeLocalizableMessage) "Text From Event" "Text")
  (Variables|Verse|SetText%20Size%20Override -1)
  (Variables|Verse|SetShow%20Border%20Override "NewEnumerator0"))

(fn SetShowBorder
  (params (Show :type Boolean))
  (Networking|SwitchHasAuthority
    (:Authority
      (Variables|Verse|SetShow%20Border%20Override (select Show "NewEnumerator1" "NewEnumerator2")))))

(fn SetViewDistance
  (bind _viewdistance (Variables|UserOptions|GetViewDistance))
  (bind _returnvalue (* _viewdistance 512.0))
  (bind _returnvalue_2 (- _returnvalue 1e+02))
  (LOD|SetMaxDrawDistance (Variables|Debug|GetStaticMeshComponent) _returnvalue)
  (LOD|SetMaxDrawDistance (Variables|Default|GetBillboard_WidgetComp) _returnvalue_2)
  (bind _billboard_widgetcomp_backface (Variables|Default|GetBillboard_WidgetComp_Backface)
    (:then
      (LOD|SetMaxDrawDistance _billboard_widgetcomp_backface _returnvalue_2))
    (:else)))

(fn ShouldInvisiblePropsBeAllowed
  (return true))

(fn UpdateBillboard
  ;[0 0.5366 1 0] This delegate is listened by the spawned Widget.
  (CallFunction|CreateorUpdateBackfaceWidget)
  ;;[0.1719 0.005371 0.005371 0.328] For jip players, if a text was overridden via verse, localization and text property replacement could happen too soon and text from event gets replaced by an outdated text right after. So we force a fetch here.
  (bind _returnvalue (Class|FortLocalizableMessageComponent|GetTextforProperty (Variables|Default|GetFortLocalizableMessage) "Text From Event"))
  (Variables|Verse|SetText%20From%20Event _returnvalue)
  (bind _details (GetDetails|GetBillboardDetails))
  (Class|BillboardWidgetComp|UpdateDetails (Variables|Default|GetBillboard_WidgetComp) _details)
  (bind _billboard_widgetcomp_backface (Variables|Default|GetBillboard_WidgetComp_Backface)
    (:then
      (Class|BillboardWidgetComp|UpdateDetails _billboard_widgetcomp_backface _details))
    (:else
      _))
  (if (Variables|Default|GetShow%20Debug%20Text)
    (bind _details_2 (GetDetails|GetBillboardDetails))
    (bind (_details_text_3_f1b35dc7484b004015ed62bbd41d9681 _details_color_text_8_391ce28e49512154106c4eb4a7edf289 _details_color_background_9_d3ae353945fa0a809706e0ba1f1f2b48 _details_textjustification_12_0727d0ab4c06ba4d2a0a479a3e482868 _details_textsize_15_5b49dc3c4500042d5c29a8a08e690f4d) (Utilities|Struct|BreakStructBillboardDetails _details_2))
    (Development|PrintString (Utilities|String|ToString(Text) (Utilities|Text|FormatText "NSLOCTEXT(\"[FCD55EAD48E9B7B93F33399017D30A1E]\", \"75C345F7415564DA24BFFD884CDFA7D5\", \"Text: {Text}, BackColor {BackColor}, Align {Align}, TextSize {TextSize}, ViewDist {ViewDist}\")" _details_text_3_f1b35dc7484b004015ed62bbd41d9681 (Utilities|String|ToString(LinearColor) _details_color_background_9_d3ae353945fa0a809706e0ba1f1f2b48) (Utilities|String|EnumtoString _details_textjustification_12_0727d0ab4c06ba4d2a0a479a3e482868) _details_textsize_15_5b49dc3c4500042d5c29a8a08e690f4d (Variables|UserOptions|GetViewDistance))) true true _details_color_text_8_391ce28e49512154106c4eb4a7edf289 1e+01)))

(fn UpdateInteractionCollision
  (params (IsInPlayMode :type Boolean))
  (bind (_backgroundcolor_r _backgroundcolor_g _backgroundcolor_b _backgroundcolor_a) (Math|Color|BreakColor (Variables|GetBackgroundColor_R)))
  (Collision|SetCollisionResponsetoChannel (Variables|Debug|GetStaticMeshComponent) "ECC_GameTraceChannel1" (select (not IsInPlayMode) "ECR_Block" (select (Math|Float|NearlyEqual(Float) _backgroundcolor_a) "ECR_Ignore" "ECR_Block")))
  (Collision|SetCollisionEnabled (Variables|Default|GetToolPickingBox) (select IsInPlayMode "NoCollision" "QueryOnly")) ; Don't want the Tool Picker box to have any \neffect during play (it can interrupt interacting\nwith other devices that are intersecting the \nbillboard.
)

(fn UpdateWidgetVisibility
  (bind _creative_enabled_component (Variables|Default|GetCreative_Enabled_Component))
  (bind _enabled (Class|CreativeEnabledComponent|IsEnabled _creative_enabled_component))
  (bind _returnvalue (Editor|InEditor))
  (bind _enabledduringphase (Variables|UserOptions|GetEnabledDuringPhase))
  (bind _returnvalue_2 (Utilities|Enum|Equal(Enum) _enabledduringphase "NewEnumerator4"))
  (bind _returnvalue_3 (and _returnvalue _returnvalue_2))
  (bind _isinplaymode (Variables|Default|GetIsInPlayMode))
  (bind _returnvalue_4 (select _returnvalue_3 "NewEnumerator1" "NewEnumerator2"))
  (bind _returnvalue_5 (select _isinplaymode "NewEnumerator0" _returnvalue_4))
  (bind _returnvalue_6 (select _enabled "NewEnumerator1" _returnvalue_5))
  ;[1 0.605 0 0.191] Widget wants to be visible during game if the device is enabled, or always during edit.
  (Networking|CanExecuteCosmeticEvents
    (:True
      (bind (_istwosided _isusingtwosidedmaterial) (CallFunction|CheckTwoSidedConfig))
      (Class|BillboardWidgetComp|UpdateVisibilityandUpdateWidget (Variables|Default|GetBillboard_WidgetComp) _returnvalue_6 _isusingtwosidedmaterial)
      (bind _billboard_widgetcomp_backface (Variables|Default|GetBillboard_WidgetComp_Backface)
        (:then
          (Class|BillboardWidgetComp|UpdateVisibilityandUpdateWidget _billboard_widgetcomp_backface _returnvalue_6 _isusingtwosidedmaterial))
        (:else)))))

(fn UpdateEditorHiddenInGame
  ;[1 0.6366 0.9138 0.1] Hiding this actor in game if we're the editor and we're not enabled during a Game Phase.
  ; (handling at this level rather than the widget level as widgets never update if their widget component is set to hidden in game, even in editor outwith Game Mode)
  (Rendering|SetActorHiddenInGame (and (not (or (Utilities|Enum|Equal(Enum) (Variables|UserOptions|GetEnabledDuringPhase) "NewEnumerator0") (Utilities|Enum|Equal(Enum) (Variables|UserOptions|GetEnabledDuringPhase) "NewEnumerator4"))) (Editor|InEditor)))
  (Development|SetHiddeninGame (Variables|Default|GetBillboard_WidgetComp) false false) ; Force Billboard Widget to not be hidden in game, or it will never update.\nThis is partly to fix broken devices in umaps (as removing the component hidden logic won't reset it on saved levels)\n. If you see this after 25.00, you can delete this.
  (bind _billboard_widgetcomp_backface (Variables|Default|GetBillboard_WidgetComp_Backface)
    (:then
      (Development|SetHiddeninGame _billboard_widgetcomp_backface false false))
    (:else)))

(fn ConstructionScript
  (|CallParentFunction))
