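; Creative Button device that broadcasts an On Interact event when allowed players press it during a minigame.
; Authority gates presses by enabled state, team/class restrictions and trigger count, runs delay and reset timers, and sets the replicated state.
; State changes drive emissive lerps and collision updates, while epoch-counted hits play sound and pulse.
; Options cover custom mesh, visibility and interaction radius with edit-time sphere preview; Verse can override text, time and trigger count.
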(event OnPropertiesLoaded(ToyOptionsComponent)
  (bind _buttonmesh (Variables|Default|GetButtonMesh))
  ;[1 1 1] Device Initialization
  (CallFunction|SetTeam (CallFunction|GetCollectingTeam))
  (Initialization|InitializeComponentWithAffects (Variables|Default|GetCreative_UseRestriction_Component) (Variables|Default|GetFortMinigameLogic) 0 0 (Creative|TeamOption|GetLegacyTeamIndex (Variables|UserOptions|Basic|GetActivatingTeam)) (Creative|ClassOption|GetLegacyClassIndex (Variables|UserOptions|GetAllowedClass)) 0 1 (Variables|UserOptions|GetInvertTeamSelection) (Variables|UserOptions|GetInvertClassSelection) false)
  (CallFunction|UpdateCollisionRadius)
  (Visibility|SetVisibilityComponentsWithCollision (Variables|Default|GetVisibleInGameComponent) (Utilities|Map|MakeMap (Variables|Default|GetButtonMesh) "QueryAndPhysics") true)
  (Visibility|SetVisiblityDuringPhase (Variables|Default|GetVisibleInGameComponent) (select (Variables|UserOptions|GetVisibleDuringGame) "NewEnumerator3" "NewEnumerator1"))
  (bind _loaded_object (Creative|Loading|AsyncLoadEditorSafe (select (CallFunction|ShouldUseCustomMesh) (Variables|UserOptions|GetCustomMesh) (Variables|Default|GetDefaultMesh)))
    (:Loaded
      (bind _asstatic_mesh (Utilities|Casting|CastToStaticMesh _loaded_object)
        (:then
          (Components|StaticMesh|SetStaticMesh _buttonmesh _asstatic_mesh)
          (Rendering|Material|SetMaterial _buttonmesh 0 (StaticMesh|GetMaterial _asstatic_mesh 0)))
        (:CastFailed
          _))
      (CallFunction|SetCurrentState "NewEnumerator0"))))

(event OnAnyPropertyChanged(ToyOptionsComponent)
  (bind _buttonmesh (Variables|Default|GetButtonMesh))
  ;[1 1 1] Device Initialization
  (CallFunction|SetTeam (CallFunction|GetCollectingTeam))
  (Initialization|InitializeComponentWithAffects (Variables|Default|GetCreative_UseRestriction_Component) (Variables|Default|GetFortMinigameLogic) 0 0 (Creative|TeamOption|GetLegacyTeamIndex (Variables|UserOptions|Basic|GetActivatingTeam)) (Creative|ClassOption|GetLegacyClassIndex (Variables|UserOptions|GetAllowedClass)) 0 1 (Variables|UserOptions|GetInvertTeamSelection) (Variables|UserOptions|GetInvertClassSelection) false)
  (CallFunction|UpdateCollisionRadius)
  (Visibility|SetVisibilityComponentsWithCollision (Variables|Default|GetVisibleInGameComponent) (Utilities|Map|MakeMap (Variables|Default|GetButtonMesh) "QueryAndPhysics") true)
  (Visibility|SetVisiblityDuringPhase (Variables|Default|GetVisibleInGameComponent) (select (Variables|UserOptions|GetVisibleDuringGame) "NewEnumerator3" "NewEnumerator1"))
  (bind _loaded_object (Creative|Loading|AsyncLoadEditorSafe (select (CallFunction|ShouldUseCustomMesh) (Variables|UserOptions|GetCustomMesh) (Variables|Default|GetDefaultMesh)))
    (:Loaded
      (bind _asstatic_mesh (Utilities|Casting|CastToStaticMesh _loaded_object)
        (:then
          (Components|StaticMesh|SetStaticMesh _buttonmesh _asstatic_mesh)
          (Rendering|Material|SetMaterial _buttonmesh 0 (StaticMesh|GetMaterial _asstatic_mesh 0)))
        (:CastFailed
          _))
      (CallFunction|SetCurrentState "NewEnumerator0"))))

(event Custom|Initialize%20Device
  (bind _buttonmesh (Variables|Default|GetButtonMesh))
  ;[1 1 1] Device Initialization
  (if (not (Variables|Default|GetbIsInitialized))
    (CallFunction|Initialize))
  (CallFunction|SetTeam (CallFunction|GetCollectingTeam))
  (Initialization|InitializeComponentWithAffects (Variables|Default|GetCreative_UseRestriction_Component) (Variables|Default|GetFortMinigameLogic) 0 0 (Creative|TeamOption|GetLegacyTeamIndex (Variables|UserOptions|Basic|GetActivatingTeam)) (Creative|ClassOption|GetLegacyClassIndex (Variables|UserOptions|GetAllowedClass)) 0 1 (Variables|UserOptions|GetInvertTeamSelection) (Variables|UserOptions|GetInvertClassSelection) false)
  (CallFunction|UpdateCollisionRadius)
  (Visibility|SetVisibilityComponentsWithCollision (Variables|Default|GetVisibleInGameComponent) (Utilities|Map|MakeMap (Variables|Default|GetButtonMesh) "QueryAndPhysics") true)
  (Visibility|SetVisiblityDuringPhase (Variables|Default|GetVisibleInGameComponent) (select (Variables|UserOptions|GetVisibleDuringGame) "NewEnumerator3" "NewEnumerator1"))
  (bind _loaded_object (Creative|Loading|AsyncLoadEditorSafe (select (CallFunction|ShouldUseCustomMesh) (Variables|UserOptions|GetCustomMesh) (Variables|Default|GetDefaultMesh)))
    (:Loaded
      (bind _asstatic_mesh (Utilities|Casting|CastToStaticMesh _loaded_object)
        (:then
          (Components|StaticMesh|SetStaticMesh _buttonmesh _asstatic_mesh)
          (Rendering|Material|SetMaterial _buttonmesh 0 (StaticMesh|GetMaterial _asstatic_mesh 0)))
        (:CastFailed
          _))
      (CallFunction|SetCurrentState "NewEnumerator0"))))

(event Custom|EnableFromEvent
  (params (Instigator :type Controller%20Object%20Reference))
  (Networking|SwitchHasAuthority
    (:Authority
      (Variables|Default|SetbIsEnabled true)
      (if (Variables|Default|GetbIsResetting)
        _
        (elif (CallFunction|CanTrigger)
          (CallFunction|SetCurrentState "NewEnumerator0"))))))

(event Custom|SetEnabled
  (params (Enabled :type Boolean))
  (if Enabled
    (Networking|SwitchHasAuthority
      (:Authority
        (Variables|Default|SetbIsEnabled true)
        (if (Variables|Default|GetbIsResetting)
          _
          (elif (CallFunction|CanTrigger)
            (CallFunction|SetCurrentState "NewEnumerator0")))))
    (else
      (Networking|SwitchHasAuthority
        (:Authority
          (Variables|Default|SetbIsEnabled false)
          (if (Variables|Default|GetbIsResetting)
            _
            (elif (CallFunction|CanTrigger)
              _
              (else
                (CallFunction|SetCurrentState "NewEnumerator2" true)))))))))

(event Custom|DisableFromEvent
  (params (Instigator :type Controller%20Object%20Reference))
  (Networking|SwitchHasAuthority
    (:Authority
      (Variables|Default|SetbIsEnabled false)
      (if (Variables|Default|GetbIsResetting)
        _
        (elif (CallFunction|CanTrigger)
          _
          (else
            (CallFunction|SetCurrentState "NewEnumerator2" true)))))))

(event Interaction|EventBlueprintOnInteract
  (params (InteractingPawn :type Fort%20Pawn%20Object%20Reference) (InteractionBeingAttempted :type EInteractionBeingAttempted%20Enum :default "FirstInteraction"))
  (Networking|SwitchHasAuthority
    (:Authority
      (CallFunction|SetInstigator InteractingPawn)
      (if (CallFunction|CanTrigger)
        (Variables|Default|SetbIsResetting true)
        (if (> (Variables|UserOptions|GetDelay) 0.0)
          (bind _returnvalue (Utilities|Time|SetTimerbyEvent (EventDispatchers|CreateEvent :!event Custom|SendTriggerEvent) (Variables|UserOptions|GetDelay)))
          (Variables|Default|SetTriggerDelayTimer _returnvalue)
          (CallFunction|SetCurrentState "NewEnumerator1" true)
          (else
            (CallFunction|SendTriggerEvent)))))))

(event MinigameStartedDelegate(FortMinigameLogic)
  (CallFunction|Initialize))

(event MinigameEndedDelegate(FortMinigameLogic)
  (Utilities|Time|ClearandInvalidateTimerbyHandle (Variables|Default|GetTriggerDelayTimer))
  (Utilities|Time|ClearandInvalidateTimerbyHandle (Variables|Default|GetResetTimer))
  (CallFunction|SetEnabled false)
  (CallFunction|SetCurrentState "NewEnumerator0")
  (Variables|Default|SetInteractTimeOverride -1.0)
  (Variables|Verse|SetInteractionStringOverride ""))

(event Custom|SendHit
  (Utilities|FlowControl|DoOnce false
    (:Completed
      (if (Networking|HasAuthority)
        (bind _output_get (Variables|Default|SetServerHitEpoch (+ (Variables|Default|GetServerHitEpoch) 1)))
        (CallFunction|Hit _output_get)
        (Utilities|FlowControl|Delay 0.2))
      (CallFunction|Hit (+ 1 (Variables|Default|GetClientHitEpoch))))))

(event Custom|SendTriggerEvent
  (if (Variables|Default|GetUseTimesCanTrigger)
    (Variables|Default|SetRemainingTimesCanTrigger (- (Variables|Default|GetRemainingTimesCanTrigger) 1)))
  (GameplayEvent|BroadcastEvent (Variables|UserOptions|Events|GetOn%20Interact) 0 (Utilities|Struct|MakeGameplayEventDeviceInteract (Variables|Default|GetTriggerInstigator)))
  (CallFunction|SendHit)
  (if (> (Variables|UserOptions|GetResetDelay) 0.0)
    (bind _returnvalue (Utilities|Time|SetTimerbyEvent (EventDispatchers|CreateEvent :!event Custom|EndReset) (Variables|UserOptions|GetResetDelay)))
    (Variables|Default|SetResetTimer _returnvalue)
    (CallFunction|SetCurrentState "NewEnumerator2" true)
    (else
      (CallFunction|EndReset))))

(event Custom|EndReset
  ;[1 1 1] End Reset
  (Variables|Default|SetbIsResetting false)
  (if (CallFunction|CanTrigger)
    (CallFunction|SetCurrentState "NewEnumerator0")
    (else
      (CallFunction|SetCurrentState "NewEnumerator2" true))))

(event OnAddedtoMinigame(FortMinigameLogic)
  (params (Minigame :type Fort%20Minigame%20Object%20Reference))
  (bind _returnvalue (Fortnite|GetFirstLocalFortPlayerController true))
  ;[0 1 0.04321 0.5] On Team and Class Changes
  (Networking|CanExecuteCosmeticEvents
    (:True
      (Utilities|IsValid Minigame
        (:Is%20Valid
          (Utilities|IsValid _returnvalue
            (:Is%20Valid
              (bind _asfort_player_state_athena (Utilities|Casting|CastToFortPlayerStateAthena (Variables|/Script/Engine.Controller|GetPlayerState _returnvalue)))
              (Player|BindEventtoOnTeamIndexChangedDelegate _asfort_player_state_athena (delegate Custom|Player%20Team%20or%20Class%20Changed))
              (Minigame|BindEventtoOnAnyLocalPlayerChangedClass Minigame (delegate Custom|Player%20Team%20or%20Class%20Changed)))))))))

(event Custom|Player%20Team%20or%20Class%20Changed
  ;[0 1 0.04321 0.5] On Team and Class Changes
  (CallFunction|UpdateCollisionRadius))

(event Custom|PulseButton
  (bind _switchmid (Variables|Default|GetSwitchMID))
  (if (not (CallFunction|ShouldUseCustomMesh))
    (Material|ReattachOrCreateMID (Variables|Default|GetButtonMesh) _switchmid)
    (Rendering|Material|GetVectorParameterValue _switchmid "Emissive" :!class /Script/Engine.MaterialInstanceDynamic)
    (bind _output_get (Variables|Default|SetStartEmissiveColor (Variables|Default|GetButtonReadyEmissiveColor)))
    (Variables|Default|SetEndEmissiveColor (Math|Color|LinearColor*Float _output_get (Variables|Default|GetPulseButtonScale)))
    (bind (_direction _lerp) (Timeline|PulseColorTL PlayFromStart)
      (:Update
        (Rendering|Material|SetVectorParameterValue (Variables|Default|GetSwitchMID) "Emissive" (Math|Color|Lerp(LinearColor) (Variables|Default|GetStartEmissiveColor) (Variables|Default|GetEndEmissiveColor) _lerp) :!class /Script/Engine.MaterialInstanceDynamic)))))

(event Custom|SetButtonEmissiveColor
  (params (Color :type Linear%20Color%20Structure))
  (bind _switchmid (Variables|Default|GetSwitchMID))
  ;[1 1 1] Client - Lerp between visual color states
  (Networking|CanExecuteCosmeticEvents
    (:True
      (if (not (CallFunction|ShouldUseCustomMesh))
        (Material|ReattachOrCreateMID (Variables|Default|GetButtonMesh) _switchmid)
        (bind _returnvalue (Rendering|Material|GetVectorParameterValue _switchmid "Emissive" :!class /Script/Engine.MaterialInstanceDynamic))
        (if (not (Math|Color|NearEqual(LinearColor) _returnvalue Color 0.01))
          (Components|Timeline|Stop (Variables|Device_Button_V2|GetPulseColorTL)))
        (Variables|Default|SetStartEmissiveColor _returnvalue)
        (Variables|Default|SetEndEmissiveColor Color)
        (if (not (Editor|InEditor))
          (bind (_direction _lerp) (Timeline|EmissiveColorLerpTL PlayFromStart)
            (:Update
              (Rendering|Material|SetVectorParameterValue (Variables|Default|GetSwitchMID) "Emissive" (Math|Color|Lerp(LinearColor) (Variables|Default|GetStartEmissiveColor) (Variables|Default|GetEndEmissiveColor) _lerp) :!class /Script/Engine.MaterialInstanceDynamic)))
          (else
            (Rendering|Material|SetVectorParameterValue (Variables|Default|GetSwitchMID) "Emissive" (Variables|Default|GetEndEmissiveColor) :!class /Script/Engine.MaterialInstanceDynamic) ; UEFN jumps straight to end\n(as it can't lerp)
          ))))))

(event Custom|Hit
  (params (Epoch :type Integer64))
  ;[1 1 1] Client - Sounds / Pulse Visuals
  (if (> Epoch (Variables|Default|GetClientHitEpoch))
    (Variables|Default|SetClientHitEpoch Epoch)
    (Networking|CanExecuteCosmeticEvents
      (:True
        (Utilities|FlowControl|Sequence
          (:then_0
            (if (Variables|UserOptions|GetTriggerSound)
              (Audio|Components|Audio|Play (Variables|Default|GetDoorbell) 0.0)))
          (:then_1
            (if (not (Components|Timeline|IsPlaying (Variables|Device_Button_V2|GetEmissiveColorLerpTL)))
              (CallFunction|PulseButton))))))))

(event Custom|ShowRangeVisualization
  (bind _returnvalue_x (Math|Vector|BreakVector (Transformation|GetWorldScale (Variables|Default|GetRangeSphere))))
  ;[1 0.639 0 0.5] Expand and Show or Shrink and Hide (or just immediately hide) the Sphere to visualize interaction range.
  (Variables|Default|SetVisualization%20Radius%20Last _returnvalue_x)
  (Variables|Default|SetVisualization%20Radius%20Desired (* (Variables|UserOptions|GetInteractionRadius) 2.0))
  (Rendering|SetVisibility (Variables|Default|GetRangeSphere) true false)
  (if (Editor|InEditor)
    (Transformation|SetWorldScale3D (Variables|Default|GetRangeSphere) (Math|Conversions|ToVector(Float) (Math|Float|Max(Float) (Variables|Default|GetVisualization%20Radius%20Desired) 0.01)))
    (else
      (Variables|Default|SetSphere%20MID (Material|GetMID (Variables|Default|GetRangeSphere)))
      (bind (_direction _lerp) (Timeline|Visualization%20Scale PlayFromStart :NewTime 1.0)
        (:Update
          (Rendering|Material|SetScalarParameterValue (Variables|Default|GetSphere%20MID) "Opacity" _lerp :!class /Script/Engine.MaterialInstanceDynamic)
          (Transformation|SetWorldScale3D (Variables|Default|GetRangeSphere) (Math|Conversions|ToVector(Float) (Math|Float|Lerp (Variables|Default|GetVisualization%20Radius%20Last) (Variables|Default|GetVisualization%20Radius%20Desired) _lerp))))
        (:Finished
          (switch Utilities|FlowControl|Switch|SwitchonETimelineDirection _direction
            (:Backward
              (Rendering|SetVisibility (Variables|Default|GetRangeSphere) false false))))))))

(event Custom|HideRangeVisualization
  (params (Immediate :type Boolean))
  (bind _returnvalue_x (Math|Vector|BreakVector (Transformation|GetWorldScale (Variables|Default|GetRangeSphere))))
  ;[1 0.639 0 0.5] Expand and Show or Shrink and Hide (or just immediately hide) the Sphere to visualize interaction range.
  (Variables|Default|SetVisualization%20Radius%20Last _returnvalue_x)
  (Variables|Default|SetVisualization%20Radius%20Desired 0.0)
  (Variables|Default|SetSphere%20MID (Material|GetMID (Variables|Default|GetRangeSphere)))
  (if (not Immediate)
    (Timeline|Visualization%20Scale PlayFromStart)
    (else
      (Timeline|Visualization%20Scale SetNewTime))))

(event OnVisibilityChanged(VisibleInGameComponent)
  (params (Visible :type Boolean))
  (CallFunction|UpdateCollisionRadius))

(event InteractInterface|EventOnInteractFocused
  (params (FortPC :type Fort%20Player%20Controller%20Object%20Reference) (ContextInfo :type Fort%20Interact%20Context%20Info%20Object%20Reference)))

(fn BlueprintCanInteract
  (params (InteractingPawn :type Fort%20Pawn%20Object%20Reference) (InteractionBeingAttempted :type EInteractionBeingAttempted%20Enum :default "FirstInteraction") (InteractionType :type TInteractionType%20Enum :default "IT_NoInteraction"))
  (Utilities|PawnIsLookingatButton InteractingPawn
    (:Yes
      (if (CallFunction|IsMinigameinProgress)
        (bind _valid_controller (CreativeUserRestrictions|ActorHasValidParticipantController (Variables|Default|GetCreative_UseRestriction_Component) (Variables|GetInteractingPawn) true false true))
        (return (and _valid_controller (Utilities|Enum|Equal(Enum) (Variables|Default|GetCurrentState) "NewEnumerator0")))
        (else
          (return (Interaction|CanEditOptionsComponent (Variables|Default|GetToyOptionsComponent) (Variables|GetInteractingPawn))))))
    (:No
      (return false))))

(fn BlueprintGetFailedInteractionString
  (params (InteractingPawn :type Fort%20Pawn%20Object%20Reference) (InteractionBeingAttempted :type EInteractionBeingAttempted%20Enum :default "FirstInteraction"))
  (return ""))

(fn BlueprintGetInteractionString
  (params (InteractingPawn :type Fort%20Pawn%20Object%20Reference) (InteractionBeingAttempted :type EInteractionBeingAttempted%20Enum :default "FirstInteraction"))
  (bind _interaction_text (Variables|UserOptions|Basic|GetInteraction%20Text))
  (if (CallFunction|IsMinigameinProgress)
    (if (Utilities|Text|TextIsEmpty (Variables|Verse|GetInteractionStringOverride))
      (if (Utilities|Text|TextIsEmpty _interaction_text)
        (return (Variables|Default|GetInteractionString_Default))
        (else
          (return _interaction_text)))
      (else
        (return (Variables|Verse|GetInteractionStringOverride))))
    (else
      (bind _returnvalue (|CallParentFunction InteractingPawn))
      (return _returnvalue))))

(fn BlueprintGetInteractionTime
  (params (InteractingPawn :type Fort%20Pawn%20Object%20Reference) (InteractionBeingAttempted :type EInteractionBeingAttempted%20Enum :default "FirstInteraction"))
  (return (Math|Float|SelectFloat (Variables|Default|GetInteractTimeOverride) (Variables|UserOptions|Basic|GetInteractTime) (>= (Variables|Default|GetInteractTimeOverride) 0)) (Class|CreativeCommonDeviceFunctionLibrary|IsinPlayMode self)))

(fn BlueprintOnLocalInteract
  (params (InteractingPawn :type Fort%20Player%20Pawn%20Object%20Reference) (InteractionBeingAttempted :type EInteractionBeingAttempted%20Enum :default "FirstInteraction"))
  (if (CallFunction|IsMinigameinProgress)
    (CallFunction|SetInstigator InteractingPawn)
    (CallFunction|SendTriggerEvent)
    (return false)
    (else
      (bind _returnvalue (Interaction|OnLocalInteract (Variables|Default|GetToyOptionsComponent) InteractingPawn))
      (return _returnvalue))))

(fn CanTrigger
  (Utilities|FlowControl|Sequence
    (:then_0
      (bind _currentminigame (Variables|/Script/FortniteGame.FortMinigameLogicComponent|GetCurrentMinigame (Variables|Default|GetFortMinigameLogic))
        (:then
          (Minigame|IsInProgress _currentminigame
            (:Yes
              (LocalVariables|SetOutCanTrigger (Math|Boolean|ANDBoolean (Variables|Default|GetbIsEnabled) (not (Variables|Default|GetbIsResetting)) (Utilities|HasTriggerCountRemaining))))))
        (:else)))
    (:then_1
      (return (LocalVariables|GetOutCanTrigger)))))

(fn GetCollectingTeam
  (return (select (Creative|TeamOption|IsTeamType (Variables|UserOptions|Basic|GetActivatingTeam)) (Creative|TeamOption|ConverttoTeam (Variables|UserOptions|Basic|GetActivatingTeam)) (Utilities|Enum|LiteralenumEFortTeam "Spectator"))))

(fn GetInteractionText
  (bind _interaction_text (Variables|UserOptions|Basic|GetInteraction%20Text))
  (if (Utilities|Text|TextIsEmpty (Variables|Verse|GetInteractionStringOverride))
    (if (Utilities|Text|TextIsEmpty _interaction_text)
      (return (Utilities|String|ToString(Text) (Variables|Default|GetInteractionString_Default)))
      (else
        (return (Utilities|String|ToString(Text) _interaction_text))))
    (else
      (return (Utilities|String|ToString(Text) (Variables|Verse|GetInteractionStringOverride))))))

(fn GetInteractionTime
  (return (Math|Float|SelectFloat (Variables|Default|GetInteractTimeOverride) (Variables|UserOptions|Basic|GetInteractTime) (>= (Variables|Default|GetInteractTimeOverride) 0))))

(fn GetRemainingNumTimesCanTrigger
  (return (Math|Integer|SelectInt (Variables|Default|GetRemainingTimesCanTrigger) 0 (Variables|Default|GetUseTimesCanTrigger))))

(fn GetTimesCanTrigger
  (return (Math|Integer|SelectInt (Variables|Verse|GetMaxTimesCanTriggerOverride) (Variables|UserOptions|GetTimesCanTrigger) (> (Variables|Verse|GetMaxTimesCanTriggerOverride) -1))))

(fn HasInteractRadius
  (return (> (Variables|UserOptions|GetInteractionRadius) 0)))

(fn Initialize
  (bind _timescantrigger (Variables|UserOptions|GetTimesCanTrigger))
  (Variables|Default|SetbIsInitialized true)
  (Variables|Verse|SetMaxTimesCanTriggerOverride -1)
  (Variables|Default|SetRemainingTimesCanTrigger _timescantrigger)
  (Variables|Default|SetUseTimesCanTrigger (!= _timescantrigger 0))
  (Variables|Default|SetbIsResetting false)
  (CallFunction|SetEnabled (Variables|UserOptions|GetEnabledAtGameStart)))

(fn IsMinigameInProgress
  (bind _currentminigame (Variables|/Script/FortniteGame.FortMinigameLogicComponent|GetCurrentMinigame (Variables|Default|GetFortMinigameLogic))
    (:then
      (return (Minigame|IsinProgress _currentminigame)))
    (:else
      (return false))))

(fn IsTargetableByLinearMover
  (return true))

(fn OnRep_CurrentState
  (CallFunction|SetMaterialBasedonCurrentState)
  (CallFunction|UpdateCollisionRadius))

(fn SetCurrentState
  (params (NewState :type EEnabledState%20Enum :default "NewEnumerator0") (bForceNetUpdate :type Boolean))
  (Networking|SwitchHasAuthority
    (:Authority
      (if (not (Creative|UEFN|IsUEFNEditModeTransitionPending 0))
        (Variables|Default|SetCurrentState NewState)
        (if bForceNetUpdate
          (Networking|ForceNetUpdate))))) ; Do not touch replicated state on clients
)

(fn SetInstigator
  (params (NewInstigator :type Actor%20Object%20Reference))
  (bind _ascontroller (Utilities|Casting|CastToController NewInstigator)
    (:then
      (Variables|Default|SetTriggerInstigator _ascontroller))
    (:CastFailed
      (bind _aspawn (Utilities|Casting|CastToPawn NewInstigator))
      (Variables|Default|SetTriggerInstigator (Pawn|GetController _aspawn)))))

(fn SetInteractionText
  (params (NewText :type Localizable%20Message%20Structure))
  (Networking|SwitchHasAuthority
    (:Authority
      (Class|FortLocalizableMessageComponent|LocalizeandSetMessage (Variables|Default|GetFortLocalizableMessage) (Variables|GetNewText) "InteractionStringOverride" "Interaction Text"))))

(fn SetInteractionTime
  (params (NewTime :type Float%20%28double-precision%29))
  (Networking|SwitchHasAuthority
    (:Authority
      (Variables|Default|SetInteractTimeOverride (Math|Float|Clamp(Float) NewTime 0.0 2e+01)))))

(fn SetMaterialBasedOnCurrentState
  (Networking|CanExecuteCosmeticEvents
    (:True
      (switch Utilities|FlowControl|Switch|SwitchonEEnabledState (Variables|Default|GetCurrentState)
        (:NewEnumerator0
          (CallFunction|SetButtonEmissiveColor (Variables|Default|GetButtonReadyEmissiveColor)))
        (:NewEnumerator1
          (CallFunction|SetButtonEmissiveColor (Variables|Default|GetButtonDelayedEmissiveColor)))
        (:NewEnumerator2
          (CallFunction|SetButtonEmissiveColor (Variables|Default|GetButtonOffEmissiveColor)))))))

(fn SetTimesCanTrigger
  (params (NewTimesCanTrigger :type Integer))
  (bind _returnvalue (Math|Integer|Clamp(Integer) NewTimesCanTrigger 0 10000))
  (Networking|SwitchHasAuthority
    (:Authority
      ;[1 1 1] treat 0 as 'infinite' if requested from verse
      (Variables|Default|SetRemainingTimesCanTrigger _returnvalue)
      (Variables|Default|SetUseTimesCanTrigger (not (== _returnvalue 0)))
      ;[1 1 1] Preserve original Max times can trigger overrride, for get purposes, independent of remaining
      (Variables|Verse|SetMaxTimesCanTriggerOverride _returnvalue))))

(fn ShouldUseCustomMesh
  (return (and (Variables|UserOptions|GetCustomMesh_Override) (Utilities|IsValidSoftObjectReference (Variables|UserOptions|GetCustomMesh)))))

(fn ShouldInvisiblePropsBeAllowed
  (return false))

(fn UpdateButtonCollision
  (if (Visibility|ShouldbeVisible (Variables|Default|GetVisibleInGameComponent))
    (Collision|SetCollisionProfileName (Variables|Default|GetButtonMesh) "FortBuildingMeshPhysics" true :!class /Script/Engine.PrimitiveComponent) ; Back to Defaults
    (else
      (Collision|SetCollisionResponsetoAllChannels (Variables|Default|GetButtonMesh) "ECR_Ignore") ; Ignore Everything
      (Collision|SetCollisionResponsetoChannel (Variables|Default|GetButtonMesh) "ECC_GameTraceChannel1" "ECR_Block") ; All collision disabled apart from Interaction
    )))

(fn UpdateCollisionRadius
  (CallFunction|UpdateButtonCollision)
  (Networking|CanExecuteCosmeticEvents
    (:True
      (LocalVariables|SetbHasRadius (CallFunction|HasInteractRadius))
      (Collision|SetCollisionEnabled (Variables|Default|GetSphereCollision) (select (Math|Boolean|ANDBoolean (Class|CreativeCommonDeviceFunctionLibrary|IsLogicinPlayMode (Variables|Default|GetFortMinigameLogic)) (LocalVariables|GetbHasRadius) (Utilities|Enum|Equal(Enum) (Variables|Default|GetCurrentState) "NewEnumerator0") (CreativeUserRestrictions|CanParticipantControllerInteract (Variables|Default|GetCreative_UseRestriction_Component) (Fortnite|GetFirstLocalFortPlayerController true) true true false)) "QueryOnly" "NoCollision"))
      (if (LocalVariables|GetbHasRadius)
        (Components|Sphere|SetSphereRadius (Variables|Default|GetSphereCollision) (* (Variables|UserOptions|GetInteractionRadius) 1e+02) false)
        (if (Class|CreativeCommonDeviceFunctionLibrary|IsinPlayMode self)
          (CallFunction|HideRangeVisualization true)
          (else
            (CallFunction|ShowRangeVisualization)))
        (else
          (CallFunction|HideRangeVisualization (Class|CreativeCommonDeviceFunctionLibrary|IsinPlayMode self)))))
    (:False
      (Collision|SetCollisionEnabled (Variables|Default|GetSphereCollision) "NoCollision") ; Server doesn't care about sphere collision as it's only used for interact (client check),\nso may as well switch it off to save overlap checks.
    )))

(fn ConstructionScript
  (Variables|/Script/Engine.SceneComponent|SetbAbsoluteScale true (Variables|Default|GetSphereCollision))
  (if (Editor|InEditor)
    (CallFunction|InitializeDevice)))
