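; Creative Capture Item Spawner device (V2) that spawns one capture item or a takeable flag for CTF-style modes.
; Items dropped on it during editing set its item list; flag definitions switch it to stand-and-interact mode, others use a minigame spawner with non-despawning pickups.
; Pickup instigator callbacks turn Taken, Dropped, Returned and Captured into tagged feedback sounds and broadcast device events.
; Authority drives spawn state, while cosmetic-only paths update preview, flag visibility and team colors.
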(event EventBeginPlay
  ;[0 1 0.1721 0.25] Initialization
  (CaptureItemSpawner|InitializeDevice)
  (CaptureItemSpawner|RefreshDeviceforPlayState)
  (Networking|SwitchHasAuthority
    (:Authority
      (bind _loaded_object (Creative|Loading|AsyncLoadEditorSafe (Variables|Constants|GetAGID_Flag))
        (:Loaded
          (bind _asfort_world_item_definition (Utilities|Casting|CastToFortWorldItemDefinition _loaded_object))
          (Variables|Default|SetFlagDefinition _asfort_world_item_definition)
          (CallFunction|RefreshItemType))))
    (:Remote
      _))
  (Networking|CanExecuteCosmeticEvents
    (:True
      (Variables|BuildingActor|SetOverridePrimitivesToExcludeFoliage (Utilities|Array|MakeArray (Variables|Default|GetBaseMesh)))
      (Initialization|InitializeComponent (Variables|Default|GetCreative_ItemPreview_Component) (Variables|Default|GetPreviewMeshScene) (Variables|Default|GetPreviewStaticMesh) (Variables|Default|GetPreviewSkeletalMesh) 0 (Variables|Default|GetPickupItemList) (Variables|Default|GetItemVariantNotifier) :!class /Game/Creative/Devices/Common/Components/Creative_ItemPreview_Component.Creative_ItemPreview_Component_C))))

(event EditorPreview|EventSetupForEditorPreview
  ;[0 1 0.1721 0.25] Initialization
  (Class|CreativeItemPreviewComponent|EditorDisplayPreview (Variables|Default|GetCreative_ItemPreview_Component) (Variables|Default|GetPreviewStaticMesh))
  (CaptureItemSpawner|InitializeDevice)
  (CaptureItemSpawner|RefreshDeviceforPlayState)
  (Networking|SwitchHasAuthority
    (:Authority
      (bind _loaded_object (Creative|Loading|AsyncLoadEditorSafe (Variables|Constants|GetAGID_Flag))
        (:Loaded
          (bind _asfort_world_item_definition (Utilities|Casting|CastToFortWorldItemDefinition _loaded_object))
          (Variables|Default|SetFlagDefinition _asfort_world_item_definition)
          (CallFunction|RefreshItemType))))
    (:Remote
      _))
  (Networking|CanExecuteCosmeticEvents
    (:True
      (Variables|BuildingActor|SetOverridePrimitivesToExcludeFoliage (Utilities|Array|MakeArray (Variables|Default|GetBaseMesh)))
      (Initialization|InitializeComponent (Variables|Default|GetCreative_ItemPreview_Component) (Variables|Default|GetPreviewMeshScene) (Variables|Default|GetPreviewStaticMesh) (Variables|Default|GetPreviewSkeletalMesh) 0 (Variables|Default|GetPickupItemList) (Variables|Default|GetItemVariantNotifier) :!class /Game/Creative/Devices/Common/Components/Creative_ItemPreview_Component.Creative_ItemPreview_Component_C))))

(event OnPropertiesLoaded(ToyOptions)
  ;[0 1 0.1721 0.25] Initialization
  (CaptureItemSpawner|InitializeDevice))

(event OnAnyPropertyChanged(ToyOptions)
  ;[0 1 0.1721 0.25] Initialization
  (CaptureItemSpawner|InitializeDevice))

(event Custom|RPC_MULTICAST%20Add%20Item
  (params (Item :type Fort%20Pickup%20Athena%20Object%20Reference))
  (Networking|CanExecuteCosmeticEvents
    (:True
      (Class|CreativeItemPreviewComponent|ItemAddedToQueueFX (Variables|Default|GetCreative_ItemPreview_Component) Item))))

(event Custom|Play%20Base%20Sound
  (params (Sound :type Sound%20Base%20Object%20Reference) (Location :type Vector))
  ;[0.9037 0.1739 1] Feedback Events
  (if (Variables|UserOptions|GetbPlayCaptureSounds)
    (Audio|PlaySoundatLocation Sound Location)))

(event EventHandleItemAdded
  (params (PlayerController :type Player%20Controller%20Object%20Reference) (ItemEntry :type Fort%20Item%20Entry%20Structure%20%28by%20ref%29))
  ;[1 0.4031 0.02478] Pickup Instigator Interface
  (|CallParentFunction PlayerController ItemEntry))

(event EventEndPlay
  (params (EndPlayReason :type EEndPlayReason%20Enum :default "Destroyed"))
  (Networking|SwitchHasAuthority
    (:Authority
      (CaptureItemSpawner|UnregisterSpecialActor))))

(event EventHandleItemRemoved
  (params (PlayerController :type Player%20Controller%20Object%20Reference) (ItemEntry :type Fort%20Item%20Entry%20Structure%20%28by%20ref%29))
  ;[1 0.4031 0.02478] Pickup Instigator Interface
  (|CallParentFunction PlayerController ItemEntry)
  (CallFunction|HandleItemDestroyedorRemoved 0 PlayerController))

(event OnComponentHit(DropArea)
  (params (HitComponent :type Primitive%20Component%20Object%20Reference) (OtherActor :type Actor%20Object%20Reference) (OtherComp :type Primitive%20Component%20Object%20Reference) (NormalImpulse :type Vector :default "0, 0, 0") (Hit :type Hit%20Result%20Structure%20%28by%20ref%29))
  (Networking|SwitchHasAuthority
    (:Authority
      (Utilities|IsValid OtherActor
        (:Is%20Valid
          (CallFunction|HandleDropAreaOverlap OtherActor))))))

(event CaptureItemSpawner|EventNotifyFeedbackEvent
  (params (EventParams :type Gameplay%20Feedback%20Event%20Params%20Structure%20%28by%20ref%29) (bFriendlyEvent :type Boolean))
  (bind (_eventtag _instigatorplayerid _eventlocation) (Utilities|Struct|BreakGameplayFeedbackEventParams EventParams))
  ;[0.9037 0.1739 1] Feedback Events
  (switch Utilities|FlowControl|Switch|SwitchonGameplayTag _eventtag
    (:Creative.Event.CaptureItem.Taken
      (CallFunction|PlayBaseSound (select bFriendlyEvent "/Game/Sounds/Creative/Modes/CTF/CTF_Pickup_Team_Cue.CTF_Pickup_Team_Cue" "/Game/Sounds/Creative/Modes/CTF/CTF_Pickup_Enemy_Cue.CTF_Pickup_Enemy_Cue") _eventlocation))
    (:Creative.Event.CaptureItem.Dropped
      (CallFunction|PlayBaseSound "/Game/Sounds/Creative/Modes/CTF/CTF_Despawn_Cue.CTF_Despawn_Cue" _eventlocation))
    (:Creative.Event.CaptureItem.Returned
      (CallFunction|PlayBaseSound (select (Math|Boolean|NotEqual(Boolean) bFriendlyEvent (CallFunction|IsTargetAffiliationFriendly)) "/Game/Sounds/Creative/Modes/CTF/CTF_Return_Team_Cue.CTF_Return_Team_Cue" "/Game/Sounds/Creative/Modes/CTF/CTF_Return_Enemy_Cue.CTF_Return_Enemy_Cue") _eventlocation))))

(event EventHandlePickupSpawned
  (params (Pickup :type Fort%20Pickup%20Object%20Reference))
  (bind (_itemdefinition _quantity _itemvariantguid) (CallFunction|GetItemEntryfromPickup Pickup))
  ;[1 0.4031 0.02478] Pickup Instigator Interface
  (|CallParentFunction Pickup)
  (Utilities|IsValid Pickup
    (:Is%20Valid
      (if (Variables|Default|GetbFirstSpawnHasOccurred)
        (CallFunction|SendCaptureItemEventatActor (Variables|Constants|GetCaptureItemReturnedTag) 0 _itemdefinition _quantity _itemvariantguid self)
        (else
          (Variables|Default|SetbFirstSpawnHasOccurred true))))))

(event OnItemAdded(PickupItemList)
  (params (ItemAdded :type Actor%20Object%20Reference))
  (Networking|SwitchHasAuthority
    (:Authority
      (CallFunction|Resave))))

(event OnItemListUpdated(PickupItemList)
  (Flag|UpdateFlagVisibility)
  (Networking|SwitchHasAuthority
    (:Authority
      (CallFunction|RefreshItemType))))

(event EventHandlePickupDropped
  (params (Pickup :type Fort%20Pickup%20Object%20Reference))
  (bind (_itemdefinition _quantity _itemvariantguid) (CallFunction|GetItemEntryfromPickup Pickup))
  ;[1 0.4031 0.02478] Pickup Instigator Interface
  ;;[1 1 1] This is only relevant for the pickup action
  (Variables|Default|SetWasDropped true)
  (|CallParentFunction Pickup)
  (Utilities|FlowControl|Sequence
    (:then_0
      (Utilities|IsValid Pickup
        (:Is%20Valid
          (Interaction|AssignOnSetPawnWhoDroppedPickup Pickup (EventDispatchers|CreateEvent :!event Custom|On%20Pickup%20Set%20Pawn%20Who%20Dropped%20Pickup))
          (CallFunction|OnPickupSetPawnWhoDroppedPickup Pickup))))
    (:then_1
      (Utilities|FlowControl|Delay 0.0)
      (bind _output_get (Variables|Default|SetPickupController (CallFunction|GetLastKnownPickupController Pickup)))
      (CallFunction|SendCaptureItemEventatActor (Variables|Constants|GetCaptureItemDroppedTag) (select (Utilities|NotEqual(Object) _output_get 0) _output_get (Variables|Default|GetLastKnownPickupController)) _itemdefinition _quantity _itemvariantguid Pickup)
      (Variables|Default|SetLastKnownPickupController 0)
      (Variables|Default|SetPickupController 0))))

(event MinigameStartedDelegate(MinigameLogic)
  (CaptureItemSpawner|RefreshDeviceforPlayState))

(event MinigameEndedDelegate(MinigameLogic)
  (CaptureItemSpawner|RefreshDeviceforPlayState))

(event Custom|OnTeamConfigChanged
  (CallFunction|UpdateDeviceColors)
  (Networking|SwitchHasAuthority
    (:Authority
      (CallFunction|UpdateInstigatorData))))

(event Custom|On%20Pickup%20Set%20Pawn%20Who%20Dropped%20Pickup
  (params (SelfActor :type Fort%20Pickup%20Object%20Reference))
  ;[1 0.4031 0.02478] Pickup Instigator Interface
  ;;[0 0 0 1 10] AFortPickup::PawnWhoDroppedPickup may be set after EventHandlePickupDropped is called, and might be null before the next tick
  (Variables|Default|SetLastKnownPickupController (CallFunction|GetLastKnownPickupController SelfActor)))

(event OnAddedtoMinigame(MinigameLogic)
  (params (Minigame :type Fort%20Minigame%20Object%20Reference))
  (CaptureItemSpawner|RefreshDeviceforMinigame Minigame)
  (CallFunction|RefreshDelegateBindings Minigame)
  ;[1 0.1082 0 0.25] Disable Collision on drop area during Edit mode, as overlaps are expensive.
  (Minigame|BindEventtoOnPlayModeChangedDelegate Minigame (delegate Custom|OnPlayModeChanged))
  (CallFunction|OnPlayModeChanged Minigame (Minigame|IsinPlayMode Minigame :!class /Script/FortniteGame.FortMinigame)))

(event EventHandlePickupClaimed
  (params (Pickup :type Fort%20Pickup%20Object%20Reference) (InteractingPawn :type Fort%20Pawn%20Object%20Reference))
  (bind (_itemdefinition _quantity _itemvariantguid) (CallFunction|GetItemEntryfromPickup Pickup))
  ;[1 0.4031 0.02478] Pickup Instigator Interface
  (|CallParentFunction Pickup InteractingPawn)
  (Pickup|OnPickup (Variables|Default|GetMinigameSpawn) Pickup InteractingPawn)
  (Utilities|IsValid InteractingPawn
    (:Is%20Valid
      (CallFunction|SendCaptureItemEventatActor (Variables|Constants|GetCaptureItemTakenTag) (Pawn|GetController InteractingPawn) _itemdefinition _quantity _itemvariantguid Pickup))))

(event Custom|OnPlayModeChanged
  (params (Minigame :type Fort%20Minigame%20Object%20Reference) (bIsInPlayMode :type Boolean))
  ;[1 0.1082 0 0.25] Disable Collision on drop area during Edit mode, as overlaps are expensive.
  (Collision|SetCollisionEnabled (Variables|Default|GetDropArea) (select bIsInPlayMode "NoCollision" "QueryOnly")))

(event OnRemovedfromMinigame(MinigameLogic)
  (params (Minigame :type Fort%20Minigame%20Object%20Reference))
  (Utilities|IsValid Minigame
    (:Is%20Valid
      (CallFunction|RefreshDelegateBindings Minigame true))))

(event EventHandlePickupDestroyed
  (params (Pickup :type Fort%20Pickup%20Object%20Reference))
  ;[1 0.4031 0.02478] Pickup Instigator Interface
  (|CallParentFunction Pickup)
  (CallFunction|HandleItemDestroyedorRemoved Pickup))

(event EventHandleItemCaptured
  (params (PlayerController :type Player%20Controller%20Object%20Reference) (ItemEntry :type Fort%20Item%20Entry%20Structure%20%28by%20ref%29))
  (bind (_itemdefinition _quantity _itemvariantguid) (CallFunction|ParseItemEntry ItemEntry))
  ;[1 0.4031 0.02478] Pickup Instigator Interface
  (|CallParentFunction PlayerController ItemEntry)
  (Variables|Default|SetbFirstSpawnHasOccurred (not (Variables|Default|GetbIsEnabled)))
  (CallFunction|SendCaptureItemEventatActor (Variables|Constants|GetCaptureItemCapturedTag) PlayerController _itemdefinition _quantity _itemvariantguid))

(event EventHandlePickupCaptured
  (params (Pickup :type Fort%20Pickup%20Object%20Reference))
  (bind _returnvalue (Pickup|GetPawnWhoDroppedPickup Pickup))
  (bind (_itemdefinition _quantity _itemvariantguid) (CallFunction|GetItemEntryfromPickup Pickup))
  ;[1 0.4031 0.02478] Pickup Instigator Interface
  (|CallParentFunction Pickup)
  (Variables|Default|SetbFirstSpawnHasOccurred (not (Variables|Default|GetbIsEnabled)))
  (Utilities|IsValid Pickup
    (:Is%20Valid
      (Utilities|IsValid _returnvalue
        (:Is%20Valid
          (CallFunction|SendCaptureItemEventatActor (Variables|Constants|GetCaptureItemCapturedTag) (Pawn|GetController _returnvalue) _itemdefinition _quantity _itemvariantguid))))))

(event Interaction|EventBlueprintOnInteract
  (params (InteractingPawn :type Fort%20Pawn%20Object%20Reference) (InteractionBeingAttempted :type EInteractionBeingAttempted%20Enum :default "FirstInteraction"))
  (bind _returnvalue (Fortnite|GetControllerfromActor InteractingPawn))
  (bind _returnvalue_2 (Pawn|GetController InteractingPawn))
  (if (Math|Boolean|ANDBoolean (Variables|Default|GetbIsFlag) (CallFunction|IsMinigameinProgress) (Variables|Default|GetbHasFlag))
    (Utilities|IsValid InteractingPawn
      (:Is%20Valid
        (bind _asfort_inventory_owner_interface (Utilities|Casting|CastToFortInventoryOwnerInterface _returnvalue_2))
        (Variables|Default|SetWasDropped false)
        ;[1 1 1] If item overflows the inventory it may be dropped at this point
        (Inventory|GiveItemtoInventoryOwner :InventoryOwner _asfort_inventory_owner_interface :ItemDefinition (Variables|Default|GetFlagDefinition) :NumberToGive 1 :bNotifyPlayer true :ItemLevel -1 :PickupInstigatorHandle (Class|FortPickupInstigator|GetTrackedIndex (Variables|Default|GetPickupItemList) 0) :bUseItemPickupAnalyticEvent false :WeaponAmmoOverride -1)
        (if (and (> (Inventory|GetItemQuantityOnPlayer _returnvalue (Variables|Default|GetFlagDefinition)) 0) (not (Variables|Default|GetWasDropped)))
          (bind _returnvalue_3 (CaptureItemSpawner|AddItemRef))
          (bind _output_get (Variables|Default|SetbHasFlag (== _returnvalue_3 0)))
          (if (not _output_get)
            (CallFunction|SendCaptureItemEventatActor :EventTag (Variables|Constants|GetCaptureItemTakenTag) :InstigatorController _returnvalue_2 :ItemDefinition (Variables|Default|GetFlagDefinition) :ItemQuantity 1 :LocationActor self))
          (else
            (CallFunction|HandleItemDestroyedorRemoved 0 _returnvalue)))))))

(event Custom|EnableFromEvent
  (params (Instigator :type Controller%20Object%20Reference))
  (if (CallFunction|IsMinigameinProgress)
    (if (not (Variables|Default|GetbIsEnabled))
      (Variables|Default|SetbIsEnabled true)
      (CallFunction|RefreshEnabledState))))

(event Custom|DisableFromEvent
  (params (Instigator :type Controller%20Object%20Reference))
  (if (CallFunction|IsMinigameinProgress)
    (if (Variables|Default|GetbIsEnabled)
      (Variables|Default|SetbIsEnabled false)
      (CallFunction|RefreshEnabledState))))

(event InteractInterface|EventOnInteractUnfocused
  (params (FortPC :type Fort%20Player%20Controller%20Object%20Reference) (NewContextInfo :type Fort%20Interact%20Context%20Info%20Object%20Reference)))

(fn ActivateFromOtherDevice
  (params (Instigator :type Controller%20Object%20Reference) (Requester :type ERequesterType%20Enum :default "NewEnumerator0"))
  (return false))

(fn AddItemToQueue
  (params (NewItem :type Fort%20Pickup%20Athena%20Object%20Reference))
  (bind _bitemadded (Class|PickupItemListComponent|AddItemtoQueue (Variables|Default|GetPickupItemList) NewItem))
  (if _bitemadded
    (CallFunction|RPCMULTICASTAddItem (Variables|GetNewItem))))

(fn BlueprintCanInteract
  (params (InteractingPawn :type Fort%20Pawn%20Object%20Reference) (InteractionBeingAttempted :type EInteractionBeingAttempted%20Enum :default "FirstInteraction") (InteractionType :type TInteractionType%20Enum :default "IT_NoInteraction"))
  (if (and (Variables|Default|GetbIsFlag) (CallFunction|IsMinigameinProgress))
    (return (Math|Boolean|ANDBoolean (CallFunction|MatchesTeamAffiliation InteractingPawn) (Variables|Default|GetbHasFlag) (CallFunction|CanTakeFlag InteractingPawn)))
    (else
      (return (Interaction|CanEditOptionsComponent (Variables|Default|GetToyOptions) InteractingPawn)))))

(fn BlueprintGetInteractionString
  (params (InteractingPawn :type Fort%20Pawn%20Object%20Reference) (InteractionBeingAttempted :type EInteractionBeingAttempted%20Enum :default "FirstInteraction"))
  (if (and (Variables|Default|GetbIsFlag) (CallFunction|IsMinigameinProgress))
    (return "NSLOCTEXT(\"[EE1DC289476BD6FDF8EA169AFE0610A1]\", \"CD0F6C484C2D877270041F85046084BC\", \"Take Flag\")")
    (else
      (bind _returnvalue (|CallParentFunction InteractingPawn))
      (return _returnvalue))))

(fn BlueprintOnLocalInteract
  (params (InteractingPawn :type Fort%20Player%20Pawn%20Object%20Reference) (InteractionBeingAttempted :type EInteractionBeingAttempted%20Enum :default "FirstInteraction"))
  (if (not (CallFunction|IsMinigameinProgress))
    (bind _returnvalue (Interaction|OnLocalInteract (Variables|Default|GetToyOptions) InteractingPawn))
    (return _returnvalue)
    (else
      (return true))))

(fn CanInteract
  (params (InteractingPawn :type Fort%20Pawn%20Object%20Reference) (InteractionBeingAttempted :type EInteractionBeingAttempted%20Enum :default "FirstInteraction") (InteractionType :type TInteractionType%20Enum :default "IT_NoInteraction"))
  (return false))

(fn DeleteExcessItemstoAdd
  (for _array_element (Variables|/Game/Creative/Devices/CaptureArea/PickupItemListComponent.PickupItemListComponent_C|GetItemsToBeAddedQueue (Variables|Default|GetPickupItemList))
    (Utilities|IsValid _array_element
      (:Is%20Valid
        (Actor|DestroyActor _array_element)))))

(fn DestroyPickups
  (params (bDestroySpawned :type Boolean :default true) (bDestroyDropped :type Boolean))
  (bind _pickupitemlist (Variables|Default|GetPickupItemList))
  (bind _output_get (LocalVariables|SetLastItemIndex (- (ItemData|GetNumItems (Variables|Default|GetPickupItemList)) 1)))
  (if (>= _output_get 0)
    (Utilities|FlowControl|Sequence
      (:then_0
        (if bDestroySpawned
          (for _index (range (+ (LocalVariables|GetLastItemIndex) 1))
            (PickupInstigator|DestroySpawnedPickups self (Class|FortPickupInstigator|GetTrackedIndex _pickupitemlist _index)))))
      (:then_1
        (if bDestroyDropped
          (for _index_2 (range (+ (LocalVariables|GetLastItemIndex) 1))
            (PickupInstigator|DestroyDroppedPickups self (Class|FortPickupInstigator|GetTrackedIndex _pickupitemlist _index_2))))))))

(fn GetDeviceColor
  (return 0))

(fn GetInstigatingController
  (return 0))

(fn GetAccentColorParam
  (Utilities|FlowControl|Sequence
    (:then_0
      (switch Utilities|FlowControl|Switch|SwitchonECreativeColorSetType (Variables|UserOptions|GetAccentColorType)
        (:NewEnumerator1
          (LocalVariables|SetParamValue (Math|Conversions|ToFloat(Byte) (Variables|UserOptions|GetAccentColor))))
        (:NewEnumerator2
          _)
        (:NewEnumerator3
          _))
      (bind _currentminigame (Variables|/Script/FortniteGame.FortMinigameLogicComponent|GetCurrentMinigame (Variables|Default|GetMinigameLogic))
        (:then
          (LocalVariables|SetParamValue (Math|Conversions|ToFloat(Integer) (Teams|GetDynamicTeamColor self _currentminigame (Variables|UserOptions|GetAccentColorType)))))
        (:else
          (if (Editor|InEditor)
            (LocalVariables|SetParamValue (Utilities|Select :Index (Variables|UserOptions|GetAccentColorType) :NewEnumerator0 0.0 :NewEnumerator1 0.0 :NewEnumerator2 (Math|Float|%25%28Float%29 (Math|Conversions|ToFloat(Integer) (+ (Math|Conversions|ToInteger(Byte) (Team|GetActorPvPTeamIndex self)) 1)) 16.0) :NewEnumerator3 0.0))))))
    (:then_1
      (return (LocalVariables|GetParamValue)))))

(fn GetItemEntryFromPickup
  (params (Pickup :type Fort%20Pickup%20Object%20Reference))
  (bind _pickup (Variables|GetPickup)
    (:then
      (return (CallFunction|ParseItemEntry (Pickup|GetItemEntry _pickup 0)) (.Quantity (CallFunction|ParseItemEntry (Pickup|GetItemEntry _pickup 0))) (.ItemVariantGuid (CallFunction|ParseItemEntry (Pickup|GetItemEntry _pickup 0)))))
    (:else
      (return 0 0 0))))

(fn GetLastKnownPickupController
  (params (Pickup :type Fort%20Pickup%20Object%20Reference))
  (Utilities|FlowControl|Sequence
    (:then_0
      (Utilities|IsValid Pickup
        (:Is%20Valid
          (bind _pawnwhodroppedpickup (Variables|/Script/FortniteGame.FortPickup|GetPawnWhoDroppedPickup Pickup)
            (:then
              (return (Pawn|GetController _pawnwhodroppedpickup)))
            (:else)))))
    (:then_1
      (return 0))))

(fn GetOverrideMeshMaterial
  (params (MeshComp :type Static%20Mesh%20Component%20Object%20Reference) (MatIdx :type Integer))
  (Utilities|FlowControl|Sequence
    (:then_0
      (if (Utilities|Equal(Object) (Variables|GetMeshComp) (Variables|Default|GetBaseMesh))
        _))
    (:then_1
      (if (Utilities|Equal(Object) (Variables|GetMeshComp) (Variables|Default|GetFlagStandMesh))
        _))
    (:then_2
      (if (Utilities|Equal(Object) (Variables|GetMeshComp) (Variables|Default|GetFlagMesh))
        (bind _flagmid (Variables|Default|GetFlagMID)
          (:then
            (return _flagmid))
          (:else))))
    (:then_3
      (return 0)))
  (bind _standbasemid (Variables|Default|GetStandBaseMID)
    (:then
      (return _standbasemid))
    (:else)))

(fn GetSpawnedPickups
  (Utilities|FlowControl|Sequence
    (:then_0
      (bind _output_get (LocalVariables|SetLastItemIndex (- (ItemData|GetNumItems (Variables|Default|GetPickupItemList)) 1)))
      (if (>= _output_get 0)
        (for _index (range (+ _output_get 1))
          (Utilities|Array|AppendArray (LocalVariables|GetDeviceSpawnedPickups) (PickupInstigator|GetSpawnedPickups self (Class|FortPickupInstigator|GetTrackedIndex (Variables|Default|GetPickupItemList) _index))))))
    (:then_1
      (return (LocalVariables|GetDeviceSpawnedPickups)))))

(fn HandleDropAreaOverlap
  (params (OverlappingActor :type Actor%20Object%20Reference))
  (if (not (CallFunction|IsMinigameinProgress))
    (bind _asfort_pickup_athena (Utilities|Casting|CastToFortPickupAthena OverlappingActor))
    (Utilities|IsPickupValid _asfort_pickup_athena (Variables|Constants|GetInvalidPickupGameplayTagContainer)
      (:IsValid
        (if (and (not (Class|PickupItemListComponent|IsIteminQueue (Variables|Default|GetPickupItemList) _asfort_pickup_athena)) (not (CallFunction|IsAnyItemStackable _asfort_pickup_athena)))
          ;[0.5168 1 0.3142] Delete items that would otherwise float in worldspace, but not be able to be picked up due to clearing the item list in the PickupItemList component.
          ;
          ; It would normally delete these items 2 seconds after the last one was dropped onto the device, but since this device only holds 1 the item list needs to be cleared whenever a new item is dropped on it at edit time.
          (CallFunction|DeleteExcessItemsToAdd)
          (Class|PickupItemListComponent|ClearItemList (Variables|Default|GetPickupItemList))
          (CallFunction|AddItemtoQueue _asfort_pickup_athena))))))

(fn HandleItemDestroyedOrRemoved
  (params (Pickup :type Fort%20Pickup%20Object%20Reference) (InstigatorController :type Controller%20Object%20Reference))
  (if (Variables|Default|GetbIsFlag)
    (Flag|RefreshFlagSpawnedState (Variables|GetPickup) (Variables|GetInstigatorController))
    (else
      (Variables|/Game/Athena/Items/Traps/MinigameSpawner/Minigame_Spawner_Component.Minigame_Spawner_Component_C|SetbBlockSpawning (or (> (Variables|CaptureItemSpawner|GetItemRefCount) 0) (not (Variables|Default|GetbIsEnabled))) (Variables|Default|GetMinigameSpawn))
      (Pickup|OnPickupDespawned (Variables|Default|GetMinigameSpawn) (Variables|GetPickup)))))

(fn InitializeDevice
  (Utilities|FlowControl|Sequence
    (:then_0
      (CallFunction|SetTeam (Creative|TeamOption|ConverttoTeam (Variables|UserOptions|GetFriendlyTeam)))
      (CallFunction|UpdateDeviceColors)
      (CallFunction|RefreshDelegateBindings (Variables|/Script/FortniteGame.FortMinigameLogicComponent|GetCurrentMinigame (Variables|Default|GetMinigameLogic))))
    (:then_1
      (CallFunction|UpdateInstigatorData))))

(fn IsAnyItemStackable
  (params (Pickup :type Fort%20Pickup%20Object%20Reference))
  (bind _pickup (Variables|GetPickup))
  (Utilities|FlowControl|Sequence
    (:then_0
      (bind _pickup_2 (Variables|GetPickup)
        (:then
          (bind _index (Utilities|FlowControl|ForLoopwithBreak 0 (- (Pickup|GetItemEntryNum _pickup_2) 1))
            (:LoopBody
              (Utilities|IsValid (Class|FortItemFunctionLibrary|BreakFortItemEntry (Pickup|GetItemEntry _pickup _index))
                (:Is%20Valid
                  (if (Class|FortItemFunctionLibrary|IsStackable (Class|FortItemFunctionLibrary|BreakFortItemEntry (Pickup|GetItemEntry _pickup _index)))
                    (LocalVariables|SetbAnyStackable true)))))))
        (:else)))
    (:then_1
      (return (LocalVariables|GetbAnyStackable)))))

(fn IsEnabled
  (return (Variables|Default|GetbIsEnabled)))

(fn IsHoldingItemForPawn
  (params (Item :type Item%20Variant%20Handle%20Structure) (Pawn :type Pawn%20Object%20Reference))
  (return false))

(fn IsMinigameInProgress
  (Utilities|FlowControl|Sequence
    (:then_0
      (bind _currentminigame (Variables|/Script/FortniteGame.FortMinigameLogicComponent|GetCurrentMinigame (Variables|Default|GetMinigameLogic))
        (:then
          (LocalVariables|SetbSuccess (Minigame|HasStarted _currentminigame :!class /Script/FortniteGame.FortMinigame)))
        (:else
          (LocalVariables|SetbSuccess (not (Editor|InEditor))))))
    (:then_1
      (return (LocalVariables|GetbSuccess)))))

(fn IsTargetAffiliationFriendly
  (return (Utilities|Enum|Equal(Enum) (select (Creative|TeamOption|IsTeamType (Variables|UserOptions|GetFriendlyTeam) "All") "Friendly" (Utilities|Select :Index (Variables|UserOptions|GetCapturedBy) :NewEnumerator0 "Friendly" :NewEnumerator1 "Hostile")) "Friendly")))

(fn MatchesTeamAffiliation
  (params (ActorToTest :type Actor%20Object%20Reference))
  (Utilities|FlowControl|Sequence
    (:then_0
      (Utilities|IsValid ActorToTest
        (:Is%20Valid
          (LocalVariables|SetbResult (or (Creative|TeamOption|IsTeamType (Variables|UserOptions|GetFriendlyTeam) "All") (Utilities|Enum|Equal(Enum) (Team|GetTeamAffiliation ActorToTest self) (Utilities|Select :Index (Variables|UserOptions|GetCapturedBy) :NewEnumerator0 "Friendly" :NewEnumerator1 "Hostile")))))))
    (:then_1
      (return (LocalVariables|GetbResult)))))

(fn MigrateLegacyUserOptions
  (bind (_accentcolorindex _targetaffiliation _accentcolor) (Class|GetClassDefaults "/CreativeCoreDevices/Device_CaptureItemSpawner_V2.Device_CaptureItemSpawner_V2_C"))
  (bind _accentcolorindex_2 (Variables|Legacy|GetAccentColorIndex))
  (bind _returnvalue (>= _accentcolorindex_2 0))
  (bind _targetaffiliation_2 (Variables|Legacy|GetTargetAffiliation))
  (Utilities|FlowControl|Sequence
    (:then_0
      ;[1 0 0.5777] TargetAffiliation -> CapturedBy
      (if (Utilities|Enum|NotEqual(Enum) _targetaffiliation_2 _targetaffiliation)
        (Variables|UserOptions|SetCapturedBy (Utilities|Select :Index _targetaffiliation_2 :Friendly "NewEnumerator0" :Neutral "NewEnumerator0" :Hostile "NewEnumerator1"))
        (Variables|Legacy|SetTargetAffiliation _targetaffiliation)
        (LocalVariables|SetbUpdateDeviceColors true)))
    (:then_1
      ;[1 0 0.5777] AccentColorIndex -> AccentColor, AccentColorType
      (if (!= (Variables|Legacy|GetAccentColorIndex) _accentcolorindex)
        (Variables|UserOptions|SetAccentColor (select _returnvalue (Utilities|Enum|BytetoEnumECreativeColor (Math|Conversions|ToByte(Integer) _accentcolorindex_2)) _accentcolor))
        (Variables|UserOptions|SetAccentColorType (select _returnvalue "NewEnumerator1" "NewEnumerator2"))
        (Variables|Legacy|SetAccentColorIndex _accentcolorindex)
        (LocalVariables|SetbUpdateDeviceColors true)))
    (:then_2
      (if (LocalVariables|GetbUpdateDeviceColors)
        (CallFunction|UpdateDeviceColors)))))

(fn OnRep_bHasFlag
  (Utilities|FlowControl|Sequence
    (:then_0
      (Flag|UpdateFlagCollision)
      (Flag|UpdateFlagVisibility))
    (:then_1
      (Networking|SwitchHasAuthority
        (:Authority
          (if (and (Variables|Default|GetbHasFlag) (CallFunction|IsMinigameinProgress))
            (CaptureItemSpawner|RegisterSpecialActor (Variables|Default|GetFlagDefinition))
            (else
              (CaptureItemSpawner|UnregisterSpecialActor))))))))

(fn OnRep_bIsFlag
  (Flag|UpdateFlagCollision)
  (Flag|UpdateFlagVisibility))

(fn ParseItemEntry
  (params (ItemEntry :type Fort%20Item%20Entry%20Structure%20%28by%20ref%29))
  (bind (_itemdefinition _itemguid _count) (Class|FortItemFunctionLibrary|BreakFortItemEntry ItemEntry))
  (bind (_itemhandle_item _itemhandle_itemvariantguid) (Utilities|Struct|BreakItemVariantHandle (Creative|ItemData|GetItemVariantHandle ItemEntry)))
  (return _itemdefinition _count _itemhandle_itemvariantguid))

(fn PickUpIItem
  (params (Instigator :type Controller%20Object%20Reference))
  (bind _fortpawn (Class|FortQuestBlueprintLibrary|GetFortPawnFromObject Instigator)
    (:Valid
      (bind _asfort_player_pawn (Utilities|Casting|CastToFortPlayerPawn _fortpawn))
      (bind _spawnedpickups (CallFunction|GetSpawnedPickups))
      (for _array_element _spawnedpickups
        (Class|FortPlayerPawn|ServerHandlePickup _asfort_player_pawn _array_element 0.0 "0, 0, 0" true)))))

(fn RefreshDelegateBindings
  (params (Minigame :type Fort%20Minigame%20Object%20Reference) (bForceUnbind :type Boolean))
  (bind _outputdelegate (EventDispatchers|CreateEvent :!event OnTeamConfigChanged))
  (bind _minigame (Variables|GetMinigame)
    (:then
      (if bForceUnbind
        _
        (elif (Utilities|Enum|Equal(Enum) (Variables|UserOptions|GetAccentColorType) "NewEnumerator1")
          _
          (else
            (Minigame|BindEventtoMinigameTeamConfigUpdated _minigame _outputdelegate))))
      (Minigame|UnbindEventfromMinigameTeamConfigUpdated _minigame _outputdelegate))
    (:else)))

(fn RefreshDeviceForMinigame
  (params (Minigame :type Fort%20Minigame%20Object%20Reference))
  (Flag|UpdateFlagVisibility)
  (CallFunction|UpdateBehaviorforGame Minigame))

(fn RefreshDeviceForPlayState
  (CaptureItemSpawner|RefreshDeviceforMinigame (Variables|/Script/FortniteGame.FortMinigameLogicComponent|GetCurrentMinigame (Variables|Default|GetMinigameLogic))))

(fn RefreshEnabledState
  (Utilities|FlowControl|Sequence
    (:then_0
      (if (Variables|Default|GetbIsFlag)
        (Flag|RefreshFlagSpawnedState 0)
        (else
          (Variables|Default|SetbHasFlag false)
          (Variables|/Game/Athena/Items/Traps/MinigameSpawner/Minigame_Spawner_Component.Minigame_Spawner_Component_C|SetbBlockSpawning (or (> (Variables|CaptureItemSpawner|GetItemRefCount) 0) (not (Variables|Default|GetbIsEnabled))) (Variables|Default|GetMinigameSpawn))
          (if (Variables|Default|GetbIsEnabled)
            (Class|MinigameSpawnerComponent|SpawnRemaining (Variables|Default|GetMinigameSpawn) true)
            (else
              (CallFunction|DestroyPickups true))))))
    (:then_1
      (if (not (Variables|Default|GetbIsEnabled))
        (Variables|Default|SetbFirstSpawnHasOccurred (> (Variables|CaptureItemSpawner|GetItemRefCount) 0))))))

(fn RefreshFlagSpawnedState
  (params (Pickup :type Fort%20Pickup%20Object%20Reference) (InstigatorController :type Controller%20Object%20Reference))
  (Utilities|FlowControl|Sequence
    (:then_0
      (LocalVariables|SetEventLocationActor (select (Utilities|IsValid (Variables|GetPickup) :!class /Script/Engine.KismetSystemLibrary) (Variables|GetPickup) self)))
    (:then_1
      (bind _output_get (Variables|Default|SetbHasFlag (and (== (Variables|CaptureItemSpawner|GetItemRefCount) 0) (Variables|Default|GetbIsEnabled))))
      (if (and _output_get (Variables|Default|GetbFirstSpawnHasOccurred))
        (CallFunction|SendCaptureItemEventatActor :EventTag (Variables|Constants|GetCaptureItemReturnedTag) :InstigatorController (Variables|GetInstigatorController) :ItemDefinition (Variables|Default|GetFlagDefinition) :ItemQuantity 1 :LocationActor (LocalVariables|GetEventLocationActor))
        (elif (Variables|Default|GetbHasFlag)
          (Variables|Default|SetbFirstSpawnHasOccurred true))))))

(fn RefreshItemType
  (Networking|SwitchHasAuthority
    (:Authority
      (Utilities|FlowControl|Sequence
        (:then_0
          (Utilities|IsValid (Variables|Default|GetFlagDefinition)
            (:Is%20Valid
              (bind _array_element (Utilities|Array|ForEachLoopwithBreak (ItemData|GetContainedItems (Variables|Default|GetPickupItemList)))
                (:LoopBody
                  (if (Utilities|Equal(Object) _array_element (Variables|Default|GetFlagDefinition))
                    (LocalVariables|SetbIsItemFlag true)))))))
        (:then_1
          (Variables|Default|SetbIsFlag (LocalVariables|GetbIsItemFlag)))))))

(fn SendCaptureItemEventAtActor
  (params (EventTag :type Gameplay%20Tag%20Structure) (InstigatorController :type Controller%20Object%20Reference) (ItemDefinition :type Item%20Definition%20Base%20Object%20Reference) (ItemQuantity :type Integer) (ItemVariantGuid :type Guid%20Structure) (LocationActor :type Actor%20Object%20Reference))
  (bind _instigatorcontroller (Variables|GetInstigatorController))
  (Utilities|FlowControl|Sequence
    (:then_0
      (bind _instigatorcontroller_2 (Variables|GetInstigatorController)
        (:then
          (bind _asfort_player_state (Utilities|Casting|CastToFortPlayerState (Variables|/Script/Engine.Controller|GetPlayerState _instigatorcontroller_2)))
          (LocalVariables|SetInstigatorPlayerId (PlayerState|GetUniqueID _asfort_player_state)))
        (:else)))
    (:then_1
      (bind _locationactor (Variables|GetLocationActor)
        (:then
          (LocalVariables|SetEventLocation (Transformation|GetActorLocation _locationactor)))
        (:else)))
    (:then_2
      (CaptureItemSpawner|SendCaptureItemFeedbackEvent (Utilities|Struct|MakeGameplayFeedbackEventParams EventTag (LocalVariables|GetInstigatorPlayerId) (LocalVariables|GetEventLocation)) (Utilities|Struct|MakeItemandCount ItemQuantity ItemDefinition ItemVariantGuid))
      (switch Utilities|FlowControl|Switch|SwitchonGameplayTag (Variables|GetEventTag)
        (:Creative.Event.CaptureItem.Taken
          (GameplayEvent|BroadcastEvent (Variables|UserOptions|Events|GetOnItemPickedUp) 0 (Utilities|Struct|MakeGameplayEventItemPickedUp _instigatorcontroller)))
        (:Creative.Event.CaptureItem.Dropped
          (GameplayEvent|BroadcastEvent (Variables|UserOptions|Events|GetOnItemDropped) 0 (Utilities|Struct|MakeGameplayEventItemDropped _instigatorcontroller)))
        (:Creative.Event.CaptureItem.Returned
          (GameplayEvent|BroadcastEvent (Variables|UserOptions|Events|GetOnItemReturned) 0 (Utilities|Struct|MakeGameplayEventItemReturned _instigatorcontroller)))
        (:Creative.Event.CaptureItem.Captured
          (GameplayEvent|BroadcastEvent (Variables|UserOptions|Events|GetOnItemCaptured) 0 (Utilities|Struct|MakeGameplayEventItemCaptured _instigatorcontroller)))))))

(fn SetTriggerCollision
  (params (On :type Boolean))
  (return false))

(fn SetupItemSpawner
  (bind _minigamespawn (Variables|Default|GetMinigameSpawn))
  (Utilities|FlowControl|Sequence
    (:then_0
      (LocalVariables|SetItemSpawnLocation (+ (Transformation|GetSocketLocation (Variables|Default|GetBaseMesh) "BaseLoc") (Math|Transform|TransformDirection (Transformation|GetActorTransform) "0, 0,64.000000"))))
    (:then_1
      (Variables|/Game/Athena/Items/Traps/MinigameSpawner/Minigame_Spawner_Component.Minigame_Spawner_Component_C|SetDelayBetweenSpawns (LocalVariables|GetTimeBetweenSpawnsSeconds) _minigamespawn)
      (Variables|/Game/Athena/Items/Traps/MinigameSpawner/Minigame_Spawner_Component.Minigame_Spawner_Component_C|SetbIsLooping (LocalVariables|GetbIsLooping) _minigamespawn)
      (Variables|/Game/Athena/Items/Traps/MinigameSpawner/Minigame_Spawner_Component.Minigame_Spawner_Component_C|SetSpareWeaponAmmo_Override (Variables|UserOptions|GetSpareWeaponAmmo_Override) _minigamespawn)
      (Variables|/Game/Athena/Items/Traps/MinigameSpawner/Minigame_Spawner_Component.Minigame_Spawner_Component_C|SetSpareWeaponAmmo (Variables|UserOptions|GetSpareWeaponAmmo) _minigamespawn)
      (Variables|/Game/Athena/Items/Traps/MinigameSpawner/Minigame_Spawner_Component.Minigame_Spawner_Component_C|SetRandomSpawnOption (LocalVariables|GetRandomOption) _minigamespawn)
      ;[1 0.6261 0] Ensure pickups can never despawn and manual spawn control is engaged
      (Variables|/Game/Athena/Items/Traps/MinigameSpawner/Minigame_Spawner_Component.Minigame_Spawner_Component_C|SetbUseManualSpawnControl true _minigamespawn)
      (Variables|/Game/Athena/Items/Traps/MinigameSpawner/Minigame_Spawner_Component.Minigame_Spawner_Component_C|SetBasePickupClass "/Game/Creative/Devices/ItemSpawnerV2/BP_Creative_Pickup_Despawn_Never.BP_Creative_Pickup_Despawn_Never_C" _minigamespawn))
    (:then_6
      (Variables|/Script/FortniteGame.FortMinigameSpawnerComponent|SetToSpawnList (ItemData|MakeSpawnParamsList (Variables|Default|GetPickupItemList) (Math|Transform|MakeTransform (LocalVariables|GetItemSpawnLocation) (Transformation|GetActorRotation))) _minigamespawn))
    (:then_7
      (Default|TryOverrideWeaponAmmoonSpawnParams (Variables|Default|GetMinigameSpawn) (Variables|UserOptions|GetInitialWeaponAmmo_Override) (Variables|UserOptions|GetInitialWeaponAmmo)))))

(fn UpdateBehaviorForGame
  (params (Minigame :type Fort%20Minigame%20Object%20Reference))
  (Flag|UpdateFlagCollision)
  (Networking|SwitchHasAuthority
    (:Authority
      (Utilities|IsValid Minigame
        (:Is%20Valid
          (if (Minigame|IsinProgress (Variables|GetMinigame))
            (CallFunction|SetupItemSpawner)
            (Variables|Default|SetbIsEnabled (Variables|UserOptions|GetbEnabledAtGameStart))
            (Variables|Default|SetbFirstSpawnHasOccurred false)
            (CallFunction|RefreshEnabledState)
            (else
              (CaptureItemSpawner|ResetItemRefCount)
              (CallFunction|RefreshItemType)
              (CaptureItemSpawner|UnregisterSpecialActor)
              (Class|MinigameSpawnerComponent|ResetSpawner (Variables|Default|GetMinigameSpawn) true))))))))

(fn UpdateDeviceColors
  (Networking|CanExecuteCosmeticEvents
    (:True
      (bind _flagmid (Variables|Default|GetFlagMID)
        (:then
          (bind _standbasemid (Variables|Default|GetStandBaseMID)
            (:then
              (Rendering|Material|SetScalarParameterValue _standbasemid "TeamColor" (CallFunction|GetAccentColorParam) :!class /Script/Engine.MaterialInstanceDynamic))
            (:else)))
        (:else)))))

(fn UpdateFlagCollision
  (Utilities|FlowControl|Sequence
    (:then_0
      (Collision|SetCollisionEnabled (Variables|Default|GetFlagStandMesh) (select (Variables|Default|GetbIsFlag) "QueryAndPhysics" "NoCollision")))
    (:then_1
      (LocalVariables|SetbShowFlag (select (CallFunction|IsMinigameinProgress) (Variables|Default|GetbHasFlag) (Variables|Default|GetbIsFlag)))
      (Collision|SetCollisionEnabled (Variables|Default|GetFlagMesh) (select (LocalVariables|GetbShowFlag) "QueryAndPhysics" "NoCollision")))))

(fn UpdateFlagVisibility
  (Networking|CanExecuteCosmeticEvents
    (:True
      (Class|CreativeItemPreviewComponent|SetEnabled (Variables|Default|GetCreative_ItemPreview_Component) (not (Variables|Default|GetbIsFlag)) :!class /Game/Creative/Devices/Common/Components/Creative_ItemPreview_Component.Creative_ItemPreview_Component_C) ; We dont want the preview to show if we're the flag
      (Utilities|FlowControl|Sequence
        (:then_0
          (UEFNConversionHelper|SetComponentHiddenEditor (Variables|Default|GetFlagStandMesh) (not (Variables|Default|GetbIsFlag))))
        (:then_1
          (bind _output_get (LocalVariables|SetbShowFlag (not (select (CallFunction|IsMinigameinProgress) (Variables|Default|GetbHasFlag) (Variables|Default|GetbIsFlag)))))
          (UEFNConversionHelper|SetComponentHiddenEditor (Variables|Default|GetFlagMesh) _output_get))))))

(fn UpdateInstigatorData
  (bind _returndroppeditems (Variables|UserOptions|GetReturnDroppedItems))
  (bind _returnvalue (Math|Conversions|ToInteger(Byte) _returndroppeditems))
  (bind _itemdroppresets (Variables|Constants|GetItemDropPresets))
  (Utilities|FlowControl|Sequence
    (:then_0
      (LocalVariables|SetItemInstigatorData)
      (Utilities|FlowControl|Sequence
        (:then_0
          (if (Utilities|Array|IsValidIndex _itemdroppresets _returnvalue)
            (LocalVariables|SetOverridePickupClass (Utilities|Array|Get(acopy) _itemdroppresets _returnvalue))
            (else
              (LocalVariables|SetOverridePickupClass "/Script/FortniteGame.FortPickupCreative"))))
        (:then_1
          (bind _output_get (LocalVariables|SetConvertedTeam (Creative|TeamOption|ConverttoTeam (Variables|UserOptions|GetFriendlyTeam))))
          (Utilities|Struct|SetmembersinPickupInstigatorData (LocalVariables|GetItemInstigatorData) _output_get (select (Creative|TeamOption|IsTeamType (Variables|UserOptions|GetFriendlyTeam) "All") "Friendly" (Utilities|Select :Index (Variables|UserOptions|GetCapturedBy) :NewEnumerator0 "Friendly" :NewEnumerator1 "Hostile")) (CallFunction|GetAccentColorParam) (Variables|UserOptions|GetScoreValue) (LocalVariables|GetOverridePickupClass)))))
    (:then_1
      (ItemData|SetBaseInstigatorData (Variables|Default|GetPickupItemList) (LocalVariables|GetItemInstigatorData)))))

(fn ConstructionScript
  (bind _self self)
  (bind _returnvalue (Transformation|GetActorLocation _self))
  (bind _x (.x _returnvalue))
  (bind _y (.y _returnvalue))
  (bind _z (.z _returnvalue))
  (bind _returnvalue_2 (Math|Float|float+float _x _y _z))
  (bind _returnvalue_3 (/ _returnvalue_2 3.0))
  (bind _returnvalue_4 (Math|Float|Truncate _returnvalue_3))
  (bind _returnvalue_5 (Utilities|String|ToString(Integer) _returnvalue_4))
  (bind _returnvalue_6 (Utilities|GetDisplayName _self))
  (bind _returnvalue_7 (Utilities|String|Append _returnvalue_6 _returnvalue_5))
  (Networking|CanExecuteCosmeticEvents
    (:True
      (Utilities|FlowControl|Sequence
        (:then_0
          (bind _returnvalue_8 (Rendering|Material|CreateDynamicMaterialInstance (Variables|Default|GetBaseMesh) 0 0 (Utilities|String|StringToName (Utilities|String|Append _returnvalue_7 "_A")) :!class /Script/Engine.PrimitiveComponent))
          (bind _output_get (Variables|Default|SetStandBaseMID _returnvalue_8))
          (Rendering|Material|SetMaterial (Variables|Default|GetFlagStandMesh) 0 _output_get))
        (:then_1
          (bind _returnvalue_9 (Rendering|Material|CreateDynamicMaterialInstance (Variables|Default|GetFlagMesh) 0 0 (Utilities|String|StringToName (Utilities|String|Append _returnvalue_7 "_B")) :!class /Script/Engine.PrimitiveComponent))
          (Variables|Default|SetFlagMID _returnvalue_9)
          (CallFunction|UpdateDeviceColors))))))
