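; Creative Channel device that relays signals between devices by named global events.
; Transmit broadcasts the broadcast name through the playspace event router and fires its On Transmit event.
; The server listens for its channel name and fires On Transmit on a match, skipping identical listen and broadcast names to prevent loops.
; Names are lowercased, Verse can override them until minigame end, and optional debug messages reach players through a message-feed ability.
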(event EventInitializeDevice
  ;[0 1 0.2942 0.2] When we update our options, update our listen for broadcast event status.
  (|CallParentFunction)
  (CallFunction|ListenForBroadcastEvent))

(event MinigameStartedDelegate(FortMinigameLogic)
  ;[0 1 0.2286 0.108] If we want to show debug messages, cache the game start time on the server so we can get a timestamp
  (Networking|SwitchHasAuthority
    (:Authority
      (if (Variables|UserOptions|Basic|GetShow%20Debug%20Message)
        (Variables|PlayerDebugMessaging|SetGame%20Start%20Time (Fortnite|Sync|Time|GetServerWorldTimeSeconds))))))

(event Custom|Transmit%20Event
  (params (Instigator :type Controller%20Object%20Reference))
  ;[0.1761 0 1 0.25] Gameplay Function - Broadcast an Event when fired.
  (bind _router (GlobalEvent|GetPlayspaceEventRouter))
  (GameplayEventRouter|BroadcastEvent _router (Utilities|Struct|MakeGameplayEventGlobalName Instigator (GlobalEvent|GetBroadcastName)) 0)
  (GlobalEvent|BroadcastDirectEvent Instigator))

(event MinigameEndedDelegate(FortMinigameLogic)
  ;[1 0.1989 0 0.199] Clear any verse overrides when the minigame ends.
  (Variables|GlobalEvent|SetVerseListenNameOverride "")
  (Variables|GlobalEvent|SetVerseBroadcastNameOverride ""))

(event Custom|Listen%20for%20Broadcast%20Event
  (bind _returnvalue (GlobalEvent|GetListenName))
  ;[0.09705 0.8298 1 0.25] Global Events. Listen for Channel devices sending the signal we're listening for, and activate if we receive it (and we're not sending on our own listen event, want to have some light infinite loop protection)
  (if (not (Editor|InEditor))
    (Networking|SwitchHasAuthority
      (:Authority
        (bind _async_task_ref (Variables|CachedReferences|GetAsync%20Task%20Ref)
          (:then
            (GameplayEventRouter|StopListeningtoEvent _async_task_ref))
          (:else
            _))
        (if (not (Utilities|String|IsEmpty (Variables|UserOptions|GetListen%20for%20Global%20Event%20Name)))
          (bind _router (GlobalEvent|GetPlayspaceEventRouter))
          (bind (_asynctaskproxy _payload) (GameplayEventRouter|StartListeningToEvent _router "/CreativeCoreDevices/Channel/Event/GameplayEvent_GlobalName.GameplayEvent_GlobalName")
            (:then
              (Variables|CachedReferences|SetAsync%20Task%20Ref _asynctaskproxy))
            (:OnEventReceived
              (if (!= _returnvalue (GlobalEvent|GetBroadcastName))
                (if (== (.EventName_6_158E86654E6C2A7E10BCB0B176D4A45B (Utilities|Struct|BreakGameplayEventGlobalName _payload)) _returnvalue)
                  (GlobalEvent|BroadcastDirectEvent (Utilities|Struct|BreakGameplayEventGlobalName _payload)))))))))))

(fn Broadcast
  (params (Instigator :type Controller%20Object%20Reference))
  (bind _router (GlobalEvent|GetPlayspaceEventRouter))
  (GameplayEventRouter|BroadcastEvent _router (Utilities|Struct|MakeGameplayEventGlobalName Instigator (GlobalEvent|GetBroadcastName)) 0)
  (Development|PrintString "VerseBroadcast"))

(fn BroadcastDebugMessage
  (params (LastMessageSuccess :type Boolean) (LastInstigator :type Controller%20Object%20Reference))
  ;[0.2086 0.3763 1 0.25] Sends a message to all players via elimination feed via a Gameplay Ability.
  (Variables|Default|SetLastMessageSuccess LastMessageSuccess)
  (Variables|Default|SetLastInstigator LastInstigator)
  (Variables|Default|SetLastActivationTime (Utilities|Time|GetGameTimeinSeconds))
  (bind _fortminigamelogic (Variables|Default|GetFortMinigameLogic)
    (:then
      (bind _currentminigame (Variables|/Script/FortniteGame.FortMinigameLogicComponent|GetCurrentMinigame _fortminigamelogic)
        (:then
          (bind _outplayers (Minigame|GetParticipatingPlayers_NotPure _currentminigame))
          (for _array_element _outplayers
            (if (not (Abilities|HasAbilitywithAnyMatchingTag (Ability|GetFortAbilitySystemComponent _array_element) (GameplayTags|MakeGameplayTagContainerfromTag "(TagName=\"Ability.Creative.MessageFeed\")") false))
              (Ability|Advanced|GrantandActivateAbilityonSelf (Ability|GetFortAbilitySystemComponent _array_element) "/Game/Creative/Devices/Common/GameplayAbilities/GA_MessageFeed.GA_MessageFeed_C"))
            (Ability|SendGameplayEventtoActor _array_element "(TagName=\"Event.Creative.Device.MessageFeed.Activate\")" (Utilities|Struct|MakeGameplayEventData "(TagName=\"\")" self))))
        (:else)))
    (:else)))

(fn BroadcastDirectEvent
  (params (Instigator :type Controller%20Object%20Reference))
  (bind _returnvalue (GameplayEvent|BroadcastEvent (Variables|UserOptions|Basic|GetOn%20Transmit) 0 (Utilities|Struct|MakeGameplayEventActivated Instigator)))
  (Utilities|ShouldShowDebug
    (:True
      (CallFunction|BroadcastDebugMessage _returnvalue Instigator))))

(fn BroadcastEventName
  (params (Instigator :type Controller%20Object%20Reference) (EventName :type String))
  (bind _router (GlobalEvent|GetPlayspaceEventRouter))
  (GameplayEventRouter|BroadcastEvent _router (Utilities|Struct|MakeGameplayEventGlobalName Instigator (select (Utilities|String|IsEmpty EventName) (GlobalEvent|GetBroadcastName) (Utilities|String|StringToName EventName))) 0)
  (Development|PrintString "VerseBroadcastEventName"))

(fn GetPlayspaceEventRouter
  ;[0.2086 0.3763 1 0.25] Returns the Event Router attached to the playspace. If we don't have a cached reference for it yet, find and cache it.
  (bind _creative_event_router_ref (Variables|Components|GetCreative%20Event%20Router%20Ref)
    (:then
      (return _creative_event_router_ref))
    (:else
      (bind _returnvalue (GameplayEventRouter|GetEventRouter "/Script/FortniteGame.FortPlayspace" self))
      (bind _output_get (Variables|Components|SetCreative%20Event%20Router%20Ref _returnvalue))
      (return _output_get))))

(fn GetBroadcastName
  (return (Utilities|String|StringToName (Utilities|String|ToLower (Utilities|String|SelectString (Variables|UserOptions|GetBroadcast%20Global%20Event%20Name) (Variables|GlobalEvent|GetVerseBroadcastNameOverride) (Utilities|String|IsEmpty (Variables|GlobalEvent|GetVerseBroadcastNameOverride)))))))

(fn GetListenName
  (return (Utilities|String|StringToName (Utilities|String|ToLower (Utilities|String|SelectString (Variables|UserOptions|GetListen%20for%20Global%20Event%20Name) (Variables|GlobalEvent|GetVerseListenNameOverride) (Utilities|String|IsEmpty (Variables|GlobalEvent|GetVerseListenNameOverride)))))))

(fn GetMessageFeedDetails
  ;[1 0.586 0 0.2] Returns what we want to show on the Message Text (requested by the Message GA)
  (Utilities|IsValid (Fortnite|Athena|GameState|GetGameStateAthena)
    (:Is%20Valid
      (bind _asfort_actor_label_override_interface (Utilities|Casting|CastToFortActorLabelOverrideInterface self))
      (return (Utilities|Text|FormatText (Variables|PlayerDebugMessaging|GetMessageText) (LabelOverride|GetDisplayLabel _asfort_actor_label_override_interface) (Utilities|Text|ToText(Float) (Math|Float|Max(Float) (- (Variables|Default|GetLastActivationTime) (Variables|PlayerDebugMessaging|GetGame%20Start%20Time))) "HalfToEven" false true 1 324 2 2) (select (Variables|Default|GetLastMessageSuccess) (Variables|PlayerDebugMessaging|GetSuccessText) (Variables|PlayerDebugMessaging|GetFailureText)) (select (or (Variables|Default|GetLastMessageSuccess) (not (== (GlobalEvent|GetBroadcastName) ""))) "" (Variables|PlayerDebugMessaging|GetFailureReason)) (select (== (GlobalEvent|GetBroadcastName) "") "" (Utilities|Text|FormatText (Variables|PlayerDebugMessaging|GetBroadcastText) (GlobalEvent|GetBroadcastName)))) (Variables|Default|GetLastInstigator) 0))))

(fn SetBroadcastNameOverride
  (params (VerseBroadcastNameOverride :type String))
  (Variables|GlobalEvent|SetVerseBroadcastNameOverride VerseBroadcastNameOverride)
  (return))

(fn SetListenNameOverride
  (params (VerseListenNameOverride :type String))
  (Variables|GlobalEvent|SetVerseListenNameOverride VerseListenNameOverride)
  (return))

(fn ConstructionScript)

(fn VerseGetBroadcastName
  (return (Utilities|String|ToString(Name) (GlobalEvent|GetBroadcastName))))

(fn VerseGetListenName
  (return (Utilities|String|ToString(Name) (GlobalEvent|GetListenName))))
