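; Class Designer V2 Creative device that defines a player class for a minigame: experience setting overrides plus an item loadout.
; Registers itself and its option component as class-slot data sources on minigame mutators, writing overrides via data registry and scalable floats.
; Minigame spawn and class delegates grant or remove items, with player sets tracking who received them.
; Pickups dropped on its area before game start fill the loadout; versioned upgrades migrate legacy options.
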(event EventBeginPlay
  ;[0 1 0.04321 0.5] Initialization
  (|CallParentFunction)
  (Networking|SwitchHasAuthority
    (:Remote
      (Initialization|InitializeComponent (Variables|Default|GetCreative_ItemPreview_Component) (Variables|Default|GetPreviewMeshScene) (Variables|Default|GetPreviewStaticMesh) (Variables|Default|GetPreviewSkeletalMesh) (Variables|Default|GetAmbient%20Loop) (Variables|Default|GetPickupItemList) (Variables|Default|GetItemVariantNotifier) :!class /Game/Creative/Devices/Common/Components/Creative_ItemPreview_Component.Creative_ItemPreview_Component_C))))

(event MinigameEndedDelegate(FortMinigameProgress)
  ;[0.5156 0 0.008952 0.2] Deconstruction - Clean Up
  (Utilities|Set|Clear (Variables|Default|GetGrantedPlayers))
  (Utilities|Set|Clear (Variables|Default|GetInitialized%20Players)))

(event Custom|Remove%20Player
  (params (UniqueNetId :type Unique%20Net%20Id%20Structure) (bIsLocalPlayer :type Boolean))
  ;[0.5156 0 0.008952 0.2] Deconstruction - Clean Up
  ;;[0.03646 0 0.000633 0] If a player is removed from the minigame, also remove them from the list of granted players.
  (Utilities|Set|Remove (Variables|Default|GetGrantedPlayers) UniqueNetId)
  (Utilities|Set|Remove (Variables|Default|GetInitialized%20Players) UniqueNetId))

(event OnRemovedfromMinigame(FortMinigameProgress)
  (params (Minigame :type Fort%20Minigame%20Object%20Reference))
  ;[0.5156 0 0.008952 0.2] Deconstruction - Clean Up
  (ClassSettings|RemoveClassSettingsfromMinigame Minigame)
  (CallFunction|BindorUnbindDelegates false Minigame))

(event OnAddedtoMinigame(FortMinigameProgress)
  (params (Minigame :type Fort%20Minigame%20Object%20Reference))
  ;[0 1 0.04321 0.5] Initialization
  (ClassSettings|HandleAddorRemoveClassSettings Minigame)
  (CallFunction|BindorUnbindDelegates true Minigame)
  (CallFunction|OnPlayModeChanged Minigame (Minigame|IsinPlayMode Minigame :!class /Script/FortniteGame.FortMinigame)))

(event Custom|Initialize%20Mesh
  ;[1 0.5502 0 0.25] Loading the Mesh if we need to (legacy)
  (if (or (Variables|UserOptions|General|GetbVisibleDuringGame) (not (Default|IsinPlayMode)))
    (bind _loaded_object (Creative|Loading|AsyncLoadEditorSafe (Variables|Default|GetMeshSoftRef))
      (:Loaded
        (bind _asstatic_mesh (Utilities|Casting|CastToStaticMesh _loaded_object))
        (Components|StaticMesh|SetStaticMesh (Variables|Default|GetBaseMesh) _asstatic_mesh) ; Clear Mesh, not needed.
      ))
    (else
      (Components|StaticMesh|SetStaticMesh (Variables|Default|GetBaseMesh) 0) ; Clear Mesh, not needed.
    )))

(event EventOnPlayModeChanged
  (params (Minigame :type Fort%20Minigame%20Object%20Reference) (bIsInPlayMode :type Boolean))
  ;[1 0.1098 0 0.251] Disable Collision on drop area during Edit mode, as overlaps are expensive.
  (|CallParentFunction Minigame bIsInPlayMode)
  (Collision|SetCollisionEnabled (Variables|Default|GetDropArea) (select bIsInPlayMode "NoCollision" "QueryOnly"))
  (CallFunction|InitializeMesh)
  (if bIsInPlayMode
    ;;[0.7135 0.07835 0 0.251] Make sure we have cached settings get updated
    (ClassSettings|RemoveClassSettingsfromMinigame Minigame)
    (ClassSettings|HandleAddorRemoveClassSettings Minigame)))

(event Custom|Player%20Spawned
  (params (SpawningPlayer :type Controller%20Object%20Reference))
  ;[0 0.8171 0.8594 0.3] Player Spawn Logic - Flag as Initialized and Attempt to Spawn Items
  (bind _playerstate (Variables|/Script/Engine.Controller|GetPlayerState SpawningPlayer)
    (:then
      (bind _asfort_player_state (Utilities|Casting|CastToFortPlayerState _playerstate))
      (if (CallFunction|IsDeviceValidForPlayerState _asfort_player_state)
        (Utilities|Set|Add (Variables|Default|GetInitialized%20Players) (PlayerState|GetUniqueID _asfort_player_state))
        (CallFunction|AttemptGrantItems (Pawn|GetControlledPawn SpawningPlayer))))
    (:else)))

(event OnAnyPropertyChanged(ToyOptionsComponent)
  ;[0 1 0.04321 0.5] Initialization
  (ClassSettings|HandleAddorRemoveClassSettings (Minigame|GetCurrentMinigame))
  (CallFunction|InitializeMesh))

(event OnPropertiesLoaded(ToyOptionsComponent)
  ;[0 1 0.04321 0.5] Initialization
  (ClassSettings|HandleAddorRemoveClassSettings (Minigame|GetCurrentMinigame))
  (CallFunction|InitializeMesh))

(event Custom|OnPlayerClassApplied
  (params (PlayerState :type Fort%20Player%20State%20Object%20Reference) (ClassSlotIndex :type Byte))
  (bind _returnvalue (PlayerState|GetCurrentPawn PlayerState))
  ;[0 0.4352 0.8594 0.2] Class Application / Removal
  (Utilities|IsValid PlayerState
    (:Is%20Valid
      (Utilities|IsValid _returnvalue
        (:Is%20Valid
          (if (Math|Boolean|ANDBoolean (not (Variables|/Script/FortniteGame.FortPawn|GetbIsDying _returnvalue)) (Utilities|IsThisClass ClassSlotIndex) (Utilities|Set|ContainsItem (Variables|Default|GetInitialized%20Players) (PlayerState|GetUniqueID PlayerState)))
            (CallFunction|AttemptGrantItems _returnvalue)))))))

(event Custom|ShowInUIWhenReceivedFrom_Func
  (params (Instigator :type Controller%20Object%20Reference))
  (bind _minigame (Minigame|GetCurrentMinigame))
  (bind _returnvalue (Minigame|GetFNEClassComponent _minigame))
  (Utilities|IsValid _minigame
    (:Is%20Valid
      (Utilities|IsValid _returnvalue
        (:Is%20Valid
          (Minigame|GameRules|SetClassSlotVisibleinUI _returnvalue (Creative|ClassOption|ConverttoClassSlotIndex (CallFunction|GetClass)) true))))))

(event OnComponentHit(DropArea)
  (params (HitComponent :type Primitive%20Component%20Object%20Reference) (OtherActor :type Actor%20Object%20Reference) (OtherComp :type Primitive%20Component%20Object%20Reference) (NormalImpulse :type Vector :default "0, 0, 0") (Hit :type Hit%20Result%20Structure%20%28by%20ref%29))
  ;[0 0.07202 1 0.5] OnComponentHit(DropArea)
  (Networking|SwitchHasAuthority
    (:Authority
      (Utilities|IsValid OtherActor
        (:Is%20Valid
          (CallFunction|HandleDropAreaOverlap OtherActor))))))

(event Custom|HideInUIWhenReceivedFrom_Func
  (params (Instigator :type Controller%20Object%20Reference))
  (bind _minigame (Minigame|GetCurrentMinigame))
  (bind _returnvalue (Minigame|GetFNEClassComponent _minigame))
  (Utilities|IsValid _minigame
    (:Is%20Valid
      (Utilities|IsValid _returnvalue
        (:Is%20Valid
          (Minigame|GameRules|SetClassSlotVisibleinUI _returnvalue (Creative|ClassOption|ConverttoClassSlotIndex (CallFunction|GetClass)) false))))))

(event OnItemAdded(PickupItemList)
  (params (ItemAdded :type Actor%20Object%20Reference))
  (CallFunction|Resave))

(event Custom|OnPlayerClassRemoved
  (params (PlayerState :type Fort%20Player%20State%20Object%20Reference) (ClassSlotIndex :type Byte))
  (bind _returnvalue (Fortnite|GetControllerfromActor PlayerState))
  ;[0 0.4352 0.8594 0.2] Class Application / Removal
  (Utilities|IsValid PlayerState
    (:Is%20Valid
      (if (Utilities|IsThisClass ClassSlotIndex)
        (Utilities|IsValid _returnvalue
          (:Is%20Valid
            (Inventory|RemoveItemsfromPlayer _returnvalue)
            (Utilities|Set|Remove (Variables|Default|GetGrantedPlayers) (PlayerState|GetUniqueID PlayerState))))))))

(event OnItemListUpdated(PickupItemList)
  (bind _minigame (Minigame|GetCurrentMinigame))
  (Utilities|IsValid _minigame
    (:Is%20Valid
      (Networking|SwitchHasAuthority
        (:Authority
          (Minigame|GameRules|RefreshGameRulesPreview _minigame))))))

(event HandleVersionUpgrades(CreativeVersion)
  (params (LastSavedVersion :type Creative%20Version%20Id%20Structure))
  (UserOptions|HandleReleaseVersionUpgrades LastSavedVersion))

(event Custom|RPC_MULTICAST%20Add%20Item
  (params (Item :type Fort%20Pickup%20Athena%20Object%20Reference))
  (Utilities|IsValid Item
    (:Is%20Valid
      (Networking|CanExecuteCosmeticEvents
        (:True
          (Class|CreativeItemPreviewComponent|ItemAddedToQueueFX (Variables|Default|GetCreative_ItemPreview_Component) Item))))))

(fn AddClassSettingsToMinigame
  (params (Minigame :type Fort%20Minigame%20Object%20Reference) (ClassSlotIndex :type Byte))
  (bind (_invincibility _startinghealth _maxhealth _startingshield _maxshields _dbnotype _nocooldowns _noswapcooldowns _infinitemagazineammo _infinitedurability _infinitecharges _infinitereserveenergy _infiniteloadedenergy _infinitebuildingresources _infinitegold _infiniteworldresources _displayemptyammoslots _movementmultiplier _applyjumppenalty _gliderredeploy _flyenabled) (Class|GetClassDefaults "/FortniteCoreUI/Settings/ExperienceSettingsComponent.ExperienceSettingsComponent_C"))
  (bind (_overridespawnimmunitytime _spawnimmunitytime) (Class|GetClassDefaults "/CreativeCoreDevices/ExperienceSettings/ExperienceSettingsComponent_Minigame.ExperienceSettingsComponent_Minigame_C"))
  (bind (_mantlingheightwaterparams _mantlingheightwatersetting) (Class|GetClassDefaults "/CreativeCoreDevices/ExperienceSettings/MantlingConversionConstants.MantlingConversionConstants_C"))
  (bind (_mantlingheightparams _mantlingheightsetting) (Class|GetClassDefaults "/CreativeCoreDevices/ExperienceSettings/MantlingConversionConstants.MantlingConversionConstants_C"))
  (bind _classslotindex (Variables|GetClassSlotIndex))
  (Utilities|FlowControl|Sequence
    (:then_0
      (bind _output_get (LocalVariables|SetSettingsMutator (ClassSettings|GetClassSettingsMutator (Variables|GetMinigame))))
      (Utilities|IsValid _output_get
        (:Is%20Valid
          (bind _returnvalue (ClassSettings|AddDataSourceforPlayerClass (LocalVariables|GetSettingsMutator) _classslotindex (Variables|Default|GetToyOptionsComponent)))
          (bind _returnvalue_2 (ClassSettings|AddDataSourceforPlayerClass (LocalVariables|GetSettingsMutator) _classslotindex self))
          (if (or _returnvalue _returnvalue_2)
            (Minigame|GameRules|RefreshGameRulesPreview (Variables|GetMinigame))))))
    (:then_1
      (bind _output_get_2 (LocalVariables|SetWeaponSettingsMutator (ClassSettings|GetWeaponSettingsMutator (Variables|GetMinigame))))
      (Utilities|IsValid _output_get_2
        (:Is%20Valid
          (Class|FortAthenaMutatorWeaponSettings|AddDataSourceforClassSlot _output_get_2 "/Game/Athena/Items/Cosmetics/Pickaxes/DefaultPickaxe.DefaultPickaxe" (Variables|GetClassSlotIndex) (Variables|Default|GetToyOptionsComponent)))))
    (:then_2
      ; ApplySharedSettings
      (Utilities|FlowControl|Sequence
        (:then_0
          (Class|FortAthenaMutatorUtilities|RemoveDataRegistryOptionsforContextObject)
          (Class|FortAthenaMutatorUtilities|ApplyActorOptionsComponenttoDataRegistry (Variables|Default|GetToyOptionsComponent)))
        (:then_1
          (ScalableFloat|SetValuewithContext _mantlingheightsetting (Class|MantlingConversionHelpers|ConvertMantlingHeightPresettoFloat (Variables|UserOptions|Player|Locomotion|GetMantlingMinimumHeight) _mantlingheightparams) (Variables|UserOptions|Overrides|GetMantlingMinimumHeight_Override))
          (ScalableFloat|SetValuewithContext _mantlingheightwatersetting (Class|MantlingConversionHelpers|ConvertMantlingHeightPresettoFloat (Variables|UserOptions|Player|Locomotion|GetMantlingMinimumHeightInWater) _mantlingheightwaterparams) (Variables|UserOptions|Overrides|GetMantlingMinimumHeightInWater_Override))
          ;; Update Core Experience Settings
          (Utilities|FlowControl|Sequence
            (:then_0
              (ScalableFloat|SetValuewithContext _invincibility (Variables|UserOptions|Player|Health|GetbInvincibility) (Variables|UserOptions|Overrides|GetbInvincibility_Override))
              (ScalableFloat|SetValuewithContext _maxhealth (Variables|UserOptions|Player|Health|GetMaxHealth) (Variables|UserOptions|Overrides|GetMaxHealth_Override))
              (ScalableFloat|SetValuewithContext _startinghealth (Variables|UserOptions|Player|Health|GetStartingHealthPercentage) (Variables|UserOptions|Overrides|GetStartingHealthPercentage_Override))
              (ScalableFloat|SetValuewithContext _maxshields (Variables|UserOptions|Player|Shields|GetMaxShields) (Variables|UserOptions|Overrides|GetMaxShields_Override))
              (ScalableFloat|SetValuewithContext _startingshield (Variables|UserOptions|Player|Shields|GetStartingShieldPercentage) (Variables|UserOptions|Overrides|GetStartingShieldPercentage_Override))
              (ScalableFloat|SetValuewithContext _dbnotype (Variables|UserOptions|Mode|Eliminations|GetDownButNotOut) (Variables|UserOptions|Overrides|GetDownButNotOut_Override))
              (ScalableFloat|SetValuewithContext _movementmultiplier (Variables|UserOptions|Player|Locomotion|GetMovementMultiplier) (Variables|UserOptions|Overrides|GetMovementMultiplier_Override))
              (ScalableFloat|SetValuewithContext _nocooldowns (Variables|UserOptions|Player|Inventory|GetbNoCooldowns) (Variables|UserOptions|Overrides|GetbNoCooldowns_Override))
              (ScalableFloat|SetValuewithContext _noswapcooldowns (Variables|UserOptions|Player|Inventory|GetbNoCooldownsAfterSwap) (Variables|UserOptions|Overrides|GetbNoCooldownsAfterSwap_Override))
              (ScalableFloat|SetValuewithContext _infinitemagazineammo (Variables|UserOptions|Player|Equipment|GetbInfiniteMagazineAmmo) (Variables|UserOptions|Overrides|GetbInfiniteMagazineAmmo_Override))
              (ScalableFloat|SetValuewithContext _infinitedurability (Variables|UserOptions|Player|Inventory|GetbInfiniteDurability) (Variables|UserOptions|Overrides|GetbInfiniteDurability_Override))
              (ScalableFloat|SetValuewithContext _infinitecharges (Variables|UserOptions|Player|Inventory|GetbInfiniteCharges) (Variables|UserOptions|Overrides|GetbInfiniteCharges_Override))
              (ScalableFloat|SetValuewithContext _infinitereserveenergy (Variables|UserOptions|Player|Inventory|GetbInfiniteReserveEnergy) (Variables|UserOptions|Overrides|GetbInfiniteReserveEnergy_Override))
              (ScalableFloat|SetValuewithContext _infiniteloadedenergy (Variables|UserOptions|Player|Inventory|GetbInfiniteLoadedEnergy) (Variables|UserOptions|Overrides|GetbInfiniteLoadedEnergy_Override))
              (ScalableFloat|SetValuewithContext _infinitebuildingresources (Variables|UserOptions|Player|Inventory|GetbInfiniteBuildingResources) (Variables|UserOptions|Overrides|GetbInfiniteBuildingResources_Override))
              (ScalableFloat|SetValuewithContext _infinitegold (Variables|UserOptions|Player|Inventory|GetbInfiniteGold) (Variables|UserOptions|Overrides|GetbInfiniteGold_Override))
              (ScalableFloat|SetValuewithContext _infiniteworldresources (Variables|UserOptions|Player|Inventory|GetbInfiniteWorldResources) (Variables|UserOptions|Overrides|GetbInfiniteWorldResources_Override))
              (ScalableFloat|SetValuewithContext _applyjumppenalty (Variables|UserOptions|Player|Locomotion|GetbApplyJumpingPenalty) (Variables|UserOptions|Overrides|GetbApplyJumpingPenalty_Override))
              (ScalableFloat|SetValuewithContext _gliderredeploy (Variables|UserOptions|Player|Locomotion|GetbGliderRedeployable) (Variables|UserOptions|Overrides|GetbGliderRedeployable_Override))
              (ScalableFloat|SetValuewithContext _flyenabled (Variables|UserOptions|Player|Locomotion|GetbFlyingEnabled) (Variables|UserOptions|Overrides|GetbFlyingEnabled_Override))
              (ScalableFloat|SetValuewithContext _displayemptyammoslots (Variables|UserOptions|Player|Equipment|GetbDisplayEmptyAmmoSlots) (Variables|UserOptions|Overrides|GetbDisplayEmptyAmmoSlots_Override)))
            (:then_1
              _))
          ;; UpdateInGameExperienceSettings
          (Utilities|FlowControl|Sequence
            (:then_0
              (ScalableFloat|SetValuewithContext _overridespawnimmunitytime (Variables|UserOptions|Mode|Spawning|GetbOverrideSpawnImmunityTime) (Variables|UserOptions|Overrides|GetbOverrideSpawnImmunityTime_Override))
              (ScalableFloat|SetValuewithContext _spawnimmunitytime (Variables|UserOptions|Mode|Spawning|GetSpawnImmunityTime) (Variables|UserOptions|Overrides|GetSpawnImmunityTime_Override)))
            (:then_1
              _)))))))

(fn AddItemToQueue
  (params (NewItem :type Fort%20Pickup%20Athena%20Object%20Reference))
  (Utilities|IsValid NewItem
    (:Is%20Valid
      (bind _bitemadded (Class|PickupItemListComponent|AddItemtoQueue (Variables|Default|GetPickupItemList) NewItem))
      (if _bitemadded
        (CallFunction|RPCMULTICASTAddItem NewItem)))))

(fn AttemptGrantItems
  (params (Possessed%20Pawn :type Pawn%20Object%20Reference) (ForceGrantItems :type Boolean))
  (Fortnite|FortShippingLog self "Device_ClassDesigner_V2 Attemptting to grant items to pawn." true)
  (bind _asfort_player_pawn (Utilities|Casting|CastToFortPlayerPawn Possessed%20Pawn)
    (:then
      (bind _asfort_player_state (Utilities|Casting|CastToFortPlayerState (Variables|/Script/Engine.Controller|GetPlayerState (Pawn|GetController _asfort_player_pawn)))
        (:then
          (if (CallFunction|IsDeviceValidForPlayerState _asfort_player_state)
            (if (ClassSettings|IsPlayerMatchingClass _asfort_player_state)
              (if (or (CallFunction|ShouldGrantedItems (PlayerState|GetUniqueID _asfort_player_state)) (Variables|GetForceGrantItems))
                (Inventory|TryGiveAllItemstoPlayer (Pawn|GetController _asfort_player_pawn))
                (Utilities|Set|Add (Variables|Default|GetGrantedPlayers) (PlayerState|GetUniqueID _asfort_player_state))
                (else
                  (Fortnite|FortShippingLog self "Device_ClassDesigner_V2 Unable to grant items. Should Grant Items is false." true)))
              (else
                _))
            (else
              _))
          (Utilities|Set|Remove (Variables|Default|GetGrantedPlayers) (PlayerState|GetUniqueID _asfort_player_state))
          (Fortnite|FortShippingLog self "Device_ClassDesigner_V2 Unable to grant items. Device is invalid for PlayerState" true))
        (:CastFailed
          (Fortnite|FortShippingLog self "Device_ClassDesigner_V2 Unable to grant items. Unable to cast to FortPlayerState" true))))
    (:CastFailed
      (Fortnite|FortShippingLog self "Device_ClassDesigner_V2 Unable to grant items. Unable to cast to Fort Player Pawn" true))))

(fn BindOrUnbindDelegates
  (params (Bind :type Boolean) (Minigame :type Fort%20Minigame%20Object%20Reference))
  (bind _outputdelegate (EventDispatchers|CreateEvent :!event Player%20Spawned))
  (bind _outputdelegate_2 (EventDispatchers|CreateEvent :!event OnPlayerClassRemoved))
  (bind _outputdelegate_3 (EventDispatchers|CreateEvent :!event OnPlayerClassApplied))
  (bind _outputdelegate_4 (EventDispatchers|CreateEvent :!event Remove%20Player))
  (bind _minigame (Variables|GetMinigame))
  (bind _minigame_2 (Variables|GetMinigame))
  (Utilities|IsValid Minigame
    (:Is%20Valid
      (if Bind
        (Minigame|BindEventtoOnMinigamePlayerorBotSpawned _minigame_2 _outputdelegate)
        (Minigame|BindEventtoOnPlayerClassRemoved _minigame_2 _outputdelegate_2)
        (Minigame|BindEventtoOnPlayerClassApplied _minigame_2 _outputdelegate_3)
        (Minigame|BindEventtoPlayerRemovedDelegate _minigame_2 _outputdelegate_4)
        (else
          (Minigame|UnbindEventfromOnMinigamePlayerorBotSpawned _minigame _outputdelegate)
          (Minigame|UnbindEventfromOnPlayerClassRemoved _minigame _outputdelegate_2)
          (Minigame|UnbindEventfromOnPlayerClassApplied _minigame _outputdelegate_3)
          (Minigame|UnbindEventfromPlayerRemovedDelegate _minigame _outputdelegate_4))))))

(fn BuildDataRegistryResolverScope
  (Utilities|FlowControl|Sequence
    (:then_0
      (bind _output_get (LocalVariables|SetSpatialInterfaceName (SpatialGameplay|GetSpatialGameplayInterfaceFNameforActor self)))
      (if (!= _output_get (LocalVariables|GetNameNone))
        (Utilities|Array|Add (LocalVariables|GetOutNameScopes) _output_get)))
    (:then_1
      (Utilities|Array|Add (LocalVariables|GetOutNameScopes) (Utilities|String|StringToName (Utilities|String|BuildString(Integer) "Class" "" (Math|Conversions|ToInteger(Byte) (Creative|ClassOption|ConverttoClassSlotIndex (CallFunction|GetClass))))))
      (LocalVariables|SetOutPriority (+ (LocalVariables|GetOutPriority) 1)))
    (:then_2
      (return (LocalVariables|GetOutNameScopes) (LocalVariables|GetOutPriority) (> (Utilities|Array|Length (LocalVariables|GetOutNameScopes)) 0)))))

(fn GetCacheContextOverride
  (return 0))

(fn GetClass
  (return (select (Variables|UserOptions|Overrides|GetClassIdentifier_Override) (Variables|UserOptions|General|GetClassIdentifier) (Utilities|Struct|MakeCreativeClassOption "NoClass"))))

(fn GetClassMembers
  (bind _minigame (Minigame|GetCurrentMinigame))
  ;[1 0 0.9139 0.5] Verse - Get all members of this designer's class.
  (Utilities|IsValid _minigame
    (:Is%20Valid
      (bind _outplayers (Minigame|GetParticipatingPlayers_NotPure _minigame))
      (for _array_element _outplayers
        (if (ClassSettings|IsPlayerMatchingClass _array_element)
          (Utilities|Array|Add (LocalVariables|GetClassMembers) (Fortnite|GetControllerfromActor _array_element))))
      (return (LocalVariables|GetClassMembers)))))

(fn GetClassSettingsMutator
  (params (Minigame :type Fort%20Minigame%20Object%20Reference))
  (Utilities|FlowControl|Sequence
    (:then_0
      (bind _output_get (LocalVariables|SetCurrentMinigame Minigame))
      (Utilities|IsValid _output_get
        (:Is%20Valid
          (bind _output_get_2 (LocalVariables|SetMutatorList (Actor|GetComponentbyClass _output_get "/Script/FortniteGame.FortMutatorListComponent")))
          (Utilities|IsValid _output_get_2
            (:Is%20Valid
              (bind _asfort_athena_mutator_player_class_settings (Utilities|Casting|CastToFortAthenaMutator_PlayerClassSettings (Modifier|GetMutatorbyClass _output_get_2 "/Script/FortniteGame.FortAthenaMutator_PlayerClassSettings")))
              (LocalVariables|SetSettingsMutator _asfort_athena_mutator_player_class_settings))))))
    (:then_1
      (return (LocalVariables|GetSettingsMutator)))))

(fn GetCreativeClassInventory
  (bind _pickupitemlist (Variables|Default|GetPickupItemList))
  (Utilities|IsValid _pickupitemlist
    (:Is%20Valid
      (return (ItemData|GetContainedItems _pickupitemlist)))
    (:Is%20Not%20Valid
      (return 0))))

(fn GetCreativeClassName
  (bind _minigame (Minigame|GetCurrentMinigame))
  (bind _returnvalue (Minigame|GetFNEClassComponent _minigame))
  (Utilities|IsValid _minigame
    (:Is%20Valid
      (Utilities|IsValid _returnvalue
        (:Is%20Valid
          (return (select (Utilities|Text|TextIsEmpty (Minigame|GameRules|GetClassSlotName _returnvalue (Creative|ClassOption|ConverttoClassSlotIndex (CallFunction|GetClass)))) "NSLOCTEXT(\"[4838684909318485EE97CAB9A53E7D2C]\", \"3312615E438F5CEF324D0EB22B551158\", \"No Class Slot Set\")" (Minigame|GameRules|GetClassSlotName _returnvalue (Creative|ClassOption|ConverttoClassSlotIndex (CallFunction|GetClass))))))
        (:Is%20Not%20Valid
          _)))
    (:Is%20Not%20Valid
      _))
  (return "NSLOCTEXT(\"[4838684909318485EE97CAB9A53E7D2C]\", \"0213F8404C999543B57D64960AAB764C\", \"No Class Info Available\")"))

(fn GetCurrentMinigame
  (return (Variables|/Script/FortniteGame.FortMinigameLogicComponent|GetCurrentMinigame (Variables|Default|GetFortMinigameProgress))))

(fn GetInventoryOverrides
  (bind _output_get (LocalVariables|SetSpawnParams (ItemData|MakeSpawnParamsList (Variables|Default|GetPickupItemList) (Math|Transform|MakeTransform "0, 0, 0"))))
  (Utilities|FlowControl|Sequence
    (:then_0
      (for _array_element _output_get
        (if (Utilities|IsValidSoftObjectReference (Utilities|Struct|BreakMinigameSpawnerSpawnParams _array_element))
          (Utilities|IsValid (Utilities|ResolveSoftReference (Utilities|Struct|BreakMinigameSpawnerSpawnParams _array_element))
            (:Is%20Valid
              (Utilities|Array|Add (LocalVariables|GetOutInventoryItems) (Utilities|Struct|MakeItemandCount (.PickupQuantity (Utilities|Struct|BreakMinigameSpawnerSpawnParams _array_element)) (Utilities|ResolveSoftReference (Utilities|Struct|BreakMinigameSpawnerSpawnParams _array_element)) (.ItemVariantGuid (Utilities|Struct|BreakMinigameSpawnerSpawnParams _array_element)))))))))
    (:then_1
      (return (LocalVariables|GetOutInventoryItems) (> (Utilities|Array|Length (LocalVariables|GetOutInventoryItems)) 0)))))

(fn GetVisibilityComponents
  (return (Utilities|Array|MakeArray (Variables|Default|GetBaseMesh) (Variables|Default|GetAmbient%20Loop))))

(fn GetVisibilityComponentsWithCollision
  (return (Utilities|Map|MakeMap (Variables|Default|GetBaseMesh) "QueryAndPhysics" (Variables|Default|GetPreviewStaticMesh) "NoCollision" (Variables|Default|GetPreviewSkeletalMesh) "NoCollision" (Variables|Default|GetAmbient%20Loop) "NoCollision")))

(fn GetWeaponSettingsMutator
  (params (Minigame :type Fort%20Minigame%20Object%20Reference))
  (bind _minigame (Variables|GetMinigame))
  (Utilities|FlowControl|Sequence
    (:then_0
      (Utilities|IsValid _minigame
        (:Is%20Valid
          (bind _output_get (LocalVariables|SetCreativeIslandMutatorList (Actor|GetComponentbyClass (Variables|/Script/FortniteGame.FortMinigame|GetVolume _minigame) "/Game/Athena/Playlists/Creative/Mutators/CreativeIslandMutatorList.CreativeIslandMutatorList_C")))
          (Utilities|IsValid _output_get
            (:Is%20Valid
              (bind _asfort_athena_mutator_weapon_settings (Utilities|Casting|CastToFortAthenaMutator_WeaponSettings (Modifier|GetMutatorbyClass _output_get "/Script/FortniteGame.FortAthenaMutator_WeaponSettings")))
              (LocalVariables|SetOutMutator _asfort_athena_mutator_weapon_settings))))))
    (:then_1
      (return (LocalVariables|GetOutMutator)))))

(fn HandleReleaseVersionUpgrades
  (params (Last%20Saved%20Version :type Creative%20Version%20Id%20Structure%20%28by%20ref%29))
  (bind (_classidentifier_classtype _classidentifier_classslot) (Utilities|Struct|BreakCreativeClassOption (Variables|GetClassIdentifier_ClassType)))
  (Utilities|FlowControl|Sequence
    (:then_0
      (CreativeVersion|CompareVersiontoTag (Variables|GetLast%20Saved%20Version) "(TagName=\"Creative.Version.Release.29.20\")"
        (:OlderVersion
          (Variables|UserOptions|Overrides|SetClassIdentifier_Override (Creative|ClassOption|IsClassType (Variables|UserOptions|General|GetClassIdentifier)))
          (Variables|SetClassIdentifier_ClassType (Utilities|Struct|MakeCreativeClassOption "ClassSlot" _classidentifier_classslot))
          (LocalVariables|SetbWasMigrated true))))
    (:then_1
      (CreativeVersion|CompareVersiontoTag (Variables|GetLast%20Saved%20Version) "(TagName=\"Creative.Version.Release.32.00\")"
        (:OlderVersion
          (bind _output_get (Variables|UserOptions|Player|Inventory|SetbInfiniteCharges (Variables|UserOptions|Player|Inventory|GetbInfiniteAmmo)))
          (Variables|UserOptions|Player|Inventory|SetbInfiniteReserveEnergy _output_get)
          (bind _output_get_2 (Variables|UserOptions|Overrides|SetbInfiniteCharges_Override (Variables|UserOptions|Overrides|GetbInfiniteAmmo_Override)))
          (Variables|UserOptions|Overrides|SetbInfiniteReserveEnergy_Override _output_get_2)
          (LocalVariables|SetbWasMigrated true))))
    (:then_2
      (CreativeVersion|CompareVersiontoTag (Variables|GetLast%20Saved%20Version) "(TagName=\"Creative.Version.Release.33.00\")"
        (:OlderVersion
          (bind _output_get_3 (Variables|UserOptions|Player|Inventory|SetbInfiniteWorldResources (Variables|Legacy|GetbInfiniteResources)))
          (bind _output_get_4 (Variables|UserOptions|Player|Inventory|SetbInfiniteBuildingResources _output_get_3))
          (Variables|UserOptions|Player|Inventory|SetbInfiniteGold _output_get_4)
          (bind _output_get_5 (Variables|UserOptions|Overrides|SetbInfiniteWorldResources_Override (Variables|Legacy|GetbInfiniteResources_Override)))
          (bind _output_get_6 (Variables|UserOptions|Overrides|SetbInfiniteBuildingResources_Override _output_get_5))
          (Variables|UserOptions|Overrides|SetbInfiniteGold_Override _output_get_6)
          (LocalVariables|SetbWasMigrated true))))
    (:then_3
      (CreativeVersion|CompareVersiontoTag (Variables|GetLast%20Saved%20Version) "(TagName=\"Creative.Version.Release.35.20\")"
        (:OlderVersion
          (Variables|UserOptions|Overrides|SetbDisplayEmptyAmmoSlots_Override true)
          (LocalVariables|SetbWasMigrated true))))
    (:then_4
      (CreativeVersion|CompareVersiontoTag (Variables|GetLast%20Saved%20Version) "(TagName=\"Creative.Version.IslandSettings.HealthShieldVariablePeriod\")"
        (:OlderVersion
          ;[1 1 1] Preserve prior experience tick period and amounts
          (Variables|UserOptions|Player|Health|SetHealthRechargePeriod 0.2)
          (Variables|UserOptions|Player|Shields|SetShieldRechargePeriod 0.2)
          (Variables|UserOptions|Player|Shields|SetOvershieldRechargePeriod 0.2)
          (Variables|UserOptions|Player|Health|SetHealthRechargeApplicationAmount (Variables|Legacy|GetHealthRechargeAmount))
          (Variables|UserOptions|Player|Shields|SetShieldRechargeApplicationAmount (Variables|Legacy|GetShieldRechargeAmount))
          (Variables|UserOptions|Player|Shields|SetOvershieldRechargeApplicationAmount (Variables|Legacy|GetOvershieldRechargeRate))
          (Variables|UserOptions|Overrides|SetOvershieldRechargeApplicationAmount_Override (or (Variables|Legacy|GetOvershieldRechargeRate_Override) (Variables|UserOptions|Overrides|GetOvershieldRechargeApplicationAmount_Override))) ; FORT-943099 the varguid was not set up correctly on first implementation, such that in UEFN the legacy data was saved directly onto the replacement variable, while FNC did not. OR the two variables to account for discrepancy.
          (LocalVariables|SetbWasMigrated true))))
    (:then_5
      (if (LocalVariables|GetbWasMigrated)
        (OptionData|RefreshPropertyOverrides (Variables|Default|GetToyOptionsComponent))))))

(fn HandleAddOrRemoveClassSettings
  (params (Minigame :type Fort%20Minigame%20Object%20Reference))
  (bind _classidentifier (CallFunction|GetClass))
  (if (Creative|ClassOption|IsClassType _classidentifier)
    (ClassSettings|AddClassSettingstoMinigame Minigame (Creative|ClassOption|ConverttoClassSlotIndex _classidentifier))
    (else
      (ClassSettings|RemoveClassSettingsfromMinigame Minigame))))

(fn HandleDropAreaOverlap
  (params (OverlappingActor :type Actor%20Object%20Reference))
  (if (not (Minigame|HasMinigameStarted))
    (bind _asfort_pickup_athena (Utilities|Casting|CastToFortPickupAthena OverlappingActor))
    (Utilities|IsPickupValid _asfort_pickup_athena (Variables|Default|GetInvalidPickupGameplayTagContainer)
      (:IsValid
        (if (not (Class|PickupItemListComponent|IsIteminQueue (Variables|Default|GetPickupItemList) _asfort_pickup_athena))
          (Inventory|AddItemtoQueue _asfort_pickup_athena))))))

(fn HasMinigameStarted
  (bind _minigame (Minigame|GetCurrentMinigame))
  (Utilities|FlowControl|Sequence
    (:then_0
      (Utilities|IsValid _minigame
        (:Is%20Valid
          (LocalVariables|SetbSuccess (Minigame|HasStarted _minigame :!class /Script/FortniteGame.FortMinigame)))))
    (:then_1
      (return (LocalVariables|GetbSuccess)))))

(fn IsDeviceValidforPlayerState
  (params (PlayerState :type Fort%20Player%20State%20Object%20Reference))
  (return (Math|Boolean|ANDBoolean (CallFunction|IsPlayeronSameIsland PlayerState) (Minigame|IsMinigameinProgress) (Creative|ClassOption|IsClassType (CallFunction|GetClass)))))

(fn IsThisClassSlot
  (params (Player%20Class :type Byte))
  (return (and (Creative|ClassOption|IsClassType (CallFunction|GetClass)) (Math|Byte|Equal(Byte) (Creative|ClassOption|ConverttoClassSlotIndex (CallFunction|GetClass)) Player%20Class))))

(fn IsAgentOfClass
  (params (Controller :type Controller%20Object%20Reference))
  (Utilities|IsValid Controller
    (:Is%20Valid
      (bind _asfort_player_state (Utilities|Casting|CastToFortPlayerState (Variables|/Script/Engine.Controller|GetPlayerState Controller))
        (:then
          (return (ClassSettings|IsPlayerMatchingClass _asfort_player_state)))
        (:CastFailed
          _)))
    (:Is%20Not%20Valid
      _))
  (return false))

(fn IsMinigameInProgress
  (bind _minigame (Minigame|GetCurrentMinigame))
  (Utilities|FlowControl|Sequence
    (:then_0
      (Utilities|IsValid _minigame
        (:Is%20Valid
          (LocalVariables|SetbSuccess (Minigame|IsinProgress _minigame)))))
    (:then_1
      (return (LocalVariables|GetbSuccess)))))

(fn IsPlayerMatchingClass
  (params (PlayerState :type Fort%20Player%20State%20Object%20Reference))
  (bind _minigame (Minigame|GetCurrentMinigame))
  (bind _returnvalue (Minigame|GetFNEClassComponent _minigame))
  (Utilities|FlowControl|Sequence
    (:then_0
      (bind _output_get (LocalVariables|SetMinigame _minigame))
      (Utilities|IsValid _output_get
        (:Is%20Valid
          (Utilities|IsValid _returnvalue
            (:Is%20Valid
              (LocalVariables|SetbSuccess (Minigame|IsPlayerUsingClassSlot _returnvalue (Variables|GetPlayerState) (Creative|ClassOption|ConverttoClassSlotIndex (CallFunction|GetClass)))))))))
    (:then_1
      (return (LocalVariables|GetbSuccess)))))

(fn IsPlayerOnSameIsland
  (params (PlayerState :type Fort%20Player%20State%20Object%20Reference))
  (bind _returnvalue (Class|CreativeBlueprintLibrary|IsActorRelevanttoActor self PlayerState))
  (return _returnvalue))

(fn IsTargetableByLinearMover
  (return true))

(fn IsVisibleDuringMinigame
  (return (and (|CallParentFunction) (Variables|UserOptions|General|GetbVisibleDuringGame))))

(fn MaybeGiveAmmo
  (params (Inventory%20Owner :type Fort%20Inventory%20Owner%20Interface%20Interface) (ItemDefinition :type Fort%20World%20Item%20Definition%20Object%20Reference) (PickupInstigatorHandle :type Integer :default -1))
  (HelperFunctions|GiveAmmotoInventoryOwner Inventory%20Owner ItemDefinition PickupInstigatorHandle (Variables|Default|GetAmmoCountMap) (Variables|Default|GetDefaultAmmoToGrant%20Const) (Variables|UserOptions|Overrides|GetSpareWeaponAmmo_Override) (Variables|UserOptions|Player|Inventory|GetSpareWeaponAmmo)))

(fn MigrateLegacyUserOptions
  (bind _sprintingspeedmultiplier (Class|GetClassDefaults "/CreativeCoreDevices/Device_ClassDesigner_V2.Device_ClassDesigner_V2_C"))
  (bind _sprintingjumpmultiplier (Class|GetClassDefaults "/CreativeCoreDevices/Device_ClassDesigner_V2.Device_ClassDesigner_V2_C"))
  (bind (_minimumledgeheight _minimumledgeheightwater _minimumledgeheight_override _minimumledgeheightwater_override) (Class|GetClassDefaults "/CreativeCoreDevices/Device_ClassDesigner_V2.Device_ClassDesigner_V2_C"))
  (bind (_mantlingheightparams _mantlingheightwaterparams) (Class|GetClassDefaults "/CreativeCoreDevices/ExperienceSettings/MantlingConversionConstants.MantlingConversionConstants_C"))
  (bind (_falldamagemultiplier _falldamagemultiplier_override) (Class|GetClassDefaults "/CreativeCoreDevices/Device_ClassDesigner_V2.Device_ClassDesigner_V2_C"))
  (bind (_nameplatedisplaydistance _nameplatedisplaydistancemeters _nameplatedisplaydistance_override _nameplatefocusangle _nameplatefocusangle_override) (Class|GetClassDefaults "/CreativeCoreDevices/Device_ClassDesigner_V2.Device_ClassDesigner_V2_C"))
  (bind _nameplatedisplaydistance_2 (Variables|Legacy|GetNameplateDisplayDistance))
  (bind _returnvalue (> _nameplatedisplaydistance_2 0))
  (bind _nameplatefocusangle_2 (Variables|Legacy|GetNameplateFocusAngle))
  (bind (_spawnimmunitytime _spawnimmunitytime_legacy _spawnimmunitytime_legacy_override) (Class|GetClassDefaults "/CreativeCoreDevices/Device_ClassDesigner_V2.Device_ClassDesigner_V2_C"))
  (bind _spawnimmunitytime_legacy_2 (Variables|Legacy|GetSpawnImmunityTime_Legacy))
  (bind _bgrantammowithweapons (Class|GetClassDefaults "/CreativeCoreDevices/Device_ClassDesigner_V2.Device_ClassDesigner_V2_C"))
  (bind (_startinghealthpercentage _startinghealthpercentage_override) (Class|GetClassDefaults "/CreativeCoreDevices/Device_ClassDesigner_V2.Device_ClassDesigner_V2_C"))
  (bind (_allowdoorbash _allowclambering _allowtacticalovershield _allowsprint _allowdoorbash_override _allowclambering_override _allowtacticalovershield_override _allowsprint_override _allowhealthshieldregenhealth _allowhealthshieldregenshield) (Class|GetClassDefaults "/CreativeCoreDevices/Device_ClassDesigner_V2.Device_ClassDesigner_V2_C"))
  (bind (_bflyenabled _bflyenabled_override) (Class|GetClassDefaults "/CreativeCoreDevices/Device_ClassDesigner_V2.Device_ClassDesigner_V2_C"))
  (bind (_bapplyjumppenalty _bapplyjumppenalty_override) (Class|GetClassDefaults "/CreativeCoreDevices/Device_ClassDesigner_V2.Device_ClassDesigner_V2_C"))
  (bind (_bgliderredeploy _bgliderredeploy_override) (Class|GetClassDefaults "/CreativeCoreDevices/Device_ClassDesigner_V2.Device_ClassDesigner_V2_C"))
  (|CallParentFunction)
  (Utilities|FlowControl|Sequence
    (:then_0
      ; MigrateSharedSettings
      (Utilities|FlowControl|Sequence
        (:then_0
          (bind _bmigrated (Creative|Migration|MigrateOptionTypeFloattoBoolean (Variables|Legacy|GetAllowDoorBash) _allowdoorbash (Variables|UserOptions|Player|Locomotion|GetbAllowShoulderBashing)))
          (bind _bmigrated_2 (Creative|Migration|MigrateOptionTypeFloattoBoolean (Variables|Legacy|GetAllowClambering) _allowclambering (Variables|UserOptions|Player|Locomotion|GetbAllowMantling)))
          (bind _bmigrated_3 (Creative|Migration|MigrateOptionTypeFloattoBoolean (Variables|Legacy|GetAllowTacticalOvershield) _allowtacticalovershield (Variables|UserOptions|Player|Shields|GetbAllowOvershield)))
          (bind _bmigrated_4 (Creative|Migration|MigrateOptionTypeFloattoBoolean (Variables|Legacy|GetAllowSprint) _allowsprint (Variables|UserOptions|Player|Locomotion|GetbAllowSprinting)))
          (bind _bmigrated_5 (Creative|Migration|MigrateOptionTypeFloattoBoolean (Variables|Legacy|GetAllowHealthShieldRegenHealth) _allowhealthshieldregenhealth (Variables|UserOptions|Player|Health|GetAllowHealthRecharge)))
          (bind _bmigrated_6 (Creative|Migration|MigrateOptionTypeFloattoBoolean (Variables|Legacy|GetAllowHealthShieldRegenShield) _allowhealthshieldregenshield (Variables|UserOptions|Player|Shields|GetAllowShieldRecharge))))
        (:then_1
          (bind _bmigrated_7 (Creative|Migration|MigrateOptionTypeBooleantoBoolean (Variables|Legacy|GetAllowDoorBash_Override) _allowdoorbash_override (Variables|UserOptions|Overrides|GetbAllowShoulderBashing_Override)))
          (bind _bmigrated_8 (Creative|Migration|MigrateOptionTypeBooleantoBoolean (Variables|Legacy|GetAllowClambering_Override) _allowclambering_override (Variables|UserOptions|Overrides|GetbAllowMantling_Override)))
          (bind _bmigrated_9 (Creative|Migration|MigrateOptionTypeBooleantoBoolean (Variables|Legacy|GetAllowTacticalOvershield_Override) _allowtacticalovershield_override (Variables|UserOptions|Overrides|GetbAllowOvershield_Override)))
          (bind _bmigrated_10 (Creative|Migration|MigrateOptionTypeBooleantoBoolean (Variables|Legacy|GetAllowSprint_Override) _allowsprint_override (Variables|UserOptions|Overrides|GetbAllowSprinting_Override)))
          (Variables|SetbWasMigrated (Math|Boolean|ORBoolean (Variables|GetbWasMigrated) (Math|Boolean|ORBoolean _bmigrated_6 _bmigrated_5 _bmigrated_4 _bmigrated_3 _bmigrated_2 _bmigrated) (Math|Boolean|ORBoolean _bmigrated_7 _bmigrated_8 _bmigrated_9 _bmigrated_10))))
        (:then_2
          _))
      ; MigrateMantlingOptions
      (Utilities|FlowControl|Sequence
        (:then_0
          (bind (_bmigrated_11 _newoptionvalue) (Class|MantlingConversionHelpers|MigrateOptionTypeFloattoMantlingHeightPreset (Variables|Legacy|GetMinimumLedgeHeight) _minimumledgeheight _mantlingheightparams))
          (if _bmigrated_11
            (Variables|UserOptions|Player|Locomotion|SetMantlingMinimumHeight (Utilities|Enum|BytetoEnumEnum_Common_MantlingLedgeHeight _newoptionvalue))
            (Variables|SetbWasMigrated true)))
        (:then_1
          (bind (_bmigrated_12 _newoptionvalue_2) (Class|MantlingConversionHelpers|MigrateOptionTypeFloattoMantlingHeightPreset (Variables|Legacy|GetMinimumLedgeHeightWater) _minimumledgeheightwater _mantlingheightwaterparams))
          (if _bmigrated_12
            (Variables|UserOptions|Player|Locomotion|SetMantlingMinimumHeightInWater (Utilities|Enum|BytetoEnumEnum_Common_MantlingLedgeHeight _newoptionvalue_2))
            (Variables|SetbWasMigrated true)))
        (:then_2
          (bind _bmigrated_13 (Creative|Migration|MigrateOptionTypeBooleantoBoolean (Variables|Legacy|GetMinimumLedgeHeight_Override) _minimumledgeheight_override (Variables|UserOptions|Overrides|GetMantlingMinimumHeight_Override)))
          (bind _bmigrated_14 (Creative|Migration|MigrateOptionTypeBooleantoBoolean (Variables|Legacy|GetMinimumLedgeHeightWater_Override) _minimumledgeheightwater_override (Variables|UserOptions|Overrides|GetMantlingMinimumHeightInWater_Override)))
          (Variables|SetbWasMigrated (Math|Boolean|ORBoolean _bmigrated_14 _bmigrated_13 (Variables|GetbWasMigrated))))
        (:then_3
          _))
      ; MigrateStartingHealth
      (Utilities|FlowControl|Sequence
        (:then_0
          (if (<= (Variables|UserOptions|Player|Health|GetStartingHealthPercentage) 0)
            (Variables|UserOptions|Player|Health|SetbInvincibility true)
            (Variables|UserOptions|Player|Health|SetStartingHealthPercentage _startinghealthpercentage)
            (Variables|SetbWasMigrated true)
            (Creative|Migration|MigrateOptionTypeBooleantoBoolean (Variables|UserOptions|Overrides|GetStartingHealthPercentage_Override) _startinghealthpercentage_override (Variables|UserOptions|Overrides|GetbInvincibility_Override))))
        (:then_1
          _))
      ; MigrateFallDamage
      (Utilities|FlowControl|Sequence
        (:then_0
          (bind _bmigrated_15 (Creative|Migration|MigrateOptionTypeFloattoBoolean (Variables|Legacy|GetFallDamageMultiplier) _falldamagemultiplier (Variables|UserOptions|Player|Locomotion|GetbFallDamage)))
          (bind _bmigrated_16 (Creative|Migration|MigrateOptionTypeBooleantoBoolean (Variables|Legacy|GetFallDamageMultiplier_Override) _falldamagemultiplier_override (Variables|UserOptions|Overrides|GetbFallDamage_Override)))
          (Variables|SetbWasMigrated (Math|Boolean|ORBoolean _bmigrated_16 _bmigrated_15 (Variables|GetbWasMigrated))))
        (:then_1
          _))
      ; MigrateSprintingSpeedMultipliers
      (Utilities|FlowControl|Sequence
        (:then_0
          (if (not (Math|Float|NearlyEqual(Float) (Variables|Legacy|GetSprintingSpeedMultiplier) _sprintingspeedmultiplier 0.001))
            (Variables|UserOptions|Player|Locomotion|SetMaxSprintingSpeedMultiplier (+ (* (- (Variables|Legacy|GetSprintingSpeedMultiplier) 1.0) 1e+01) 1.0))
            (Variables|Legacy|SetSprintingSpeedMultiplier _sprintingspeedmultiplier)
            (Variables|SetbWasMigrated true)))
        (:then_1
          (if (not (Math|Float|NearlyEqual(Float) (Variables|Legacy|GetSprintingJumpMultiplier) _sprintingjumpmultiplier 0.001))
            (Variables|UserOptions|Player|Locomotion|SetMaxSprintingJumpMultiplier (+ (* (- (Variables|Legacy|GetSprintingJumpMultiplier) 1.0) 1e+01) 1.0))
            (Variables|Legacy|SetSprintingJumpMultiplier _sprintingjumpmultiplier)
            (Variables|SetbWasMigrated true)))
        (:then_2
          (bind _bmigrated_17 (Creative|Migration|MigrateOptionTypeBooleantoBoolean (Variables|Legacy|GetSprintingJumpMultiplier_Override) (Class|GetClassDefaults "/CreativeCoreDevices/Device_ClassDesigner_V2.Device_ClassDesigner_V2_C") (Variables|UserOptions|Overrides|GetMaxSprintingJumpMultiplier_Override)))
          (Variables|SetbWasMigrated (or _bmigrated_17 (Variables|GetbWasMigrated))))
        (:then_3
          (bind _bmigrated_18 (Creative|Migration|MigrateOptionTypeBooleantoBoolean (Variables|Legacy|GetSprintingSpeedMultiplier_Override) (Class|GetClassDefaults "/CreativeCoreDevices/Device_ClassDesigner_V2.Device_ClassDesigner_V2_C") (Variables|UserOptions|Overrides|GetMaxSprintingSpeedMultiplier_Override)))
          (Variables|SetbWasMigrated (or _bmigrated_18 (Variables|GetbWasMigrated))))
        (:then_4
          _)))
    (:then_1
      ; MigrateNameplateSettings
      (Utilities|FlowControl|Sequence
        (:then_0
          (bind _bmigrated_19 (Creative|Migration|MigrateOptionTypeBooleantoBoolean (Variables|Legacy|GetNameplateDisplayDistance_Override) _nameplatedisplaydistance_override (Variables|UserOptions|Overrides|GetNameplateDisplayDistanceMeters_Override)))
          (if _bmigrated_19
            (Variables|UserOptions|Overrides|SetbLimitNamePlateMaxDistance_Override (Variables|UserOptions|Overrides|GetNameplateDisplayDistanceMeters_Override))
            (Variables|SetbWasMigrated true)))
        (:then_1
          (if (!= _nameplatedisplaydistance _nameplatedisplaydistance_2)
            (Variables|UserOptions|InGameInterface|Nameplate|SetbLimitNamePlateMaxDistance _returnvalue)
            (Variables|UserOptions|InGameInterface|Nameplate|SetNameplateDisplayDistanceMeters (select _returnvalue (/ _nameplatedisplaydistance_2 1e+02) _nameplatedisplaydistancemeters))
            (Variables|Legacy|SetNameplateDisplayDistance _nameplatedisplaydistance)
            (Variables|SetbWasMigrated true)))
        (:then_2
          (if (!= _nameplatefocusangle _nameplatefocusangle_2)
            (Variables|UserOptions|InGameInterface|Nameplate|SetNameplateFocusAngleDegrees (Math|Conversions|ToFloat(Integer) (Math|Float|Round (Math|Trig|RadiansToDegrees _nameplatefocusangle_2))))
            (Variables|Legacy|SetNameplateFocusAngle _nameplatefocusangle)
            (Variables|SetbWasMigrated true)))
        (:then_3
          (bind _bmigrated_20 (Creative|Migration|MigrateOptionTypeBooleantoBoolean (Variables|Legacy|GetNameplateFocusAngle_Override) _nameplatefocusangle_override (Variables|UserOptions|Overrides|GetNameplateFocusAngleDegrees_Override)))
          (if _bmigrated_20
            (Variables|SetbWasMigrated true)))
        (:then_4
          (if (< (Variables|UserOptions|InGameInterface|Nameplate|GetNameplateFocusTime) 0)
            (Variables|UserOptions|InGameInterface|Nameplate|SetNameplateFocusTime 0.0)
            (Variables|SetbWasMigrated true)))
        (:then_5
          _))
      ; MigrateSpawnImmunityTime
      (Utilities|FlowControl|Sequence
        (:then_0
          (if (!= _spawnimmunitytime_legacy _spawnimmunitytime_legacy_2)
            (bind _output_get (Variables|UserOptions|Mode|Spawning|SetbOverrideSpawnImmunityTime (>= _spawnimmunitytime_legacy_2 0)))
            (Variables|UserOptions|Mode|Spawning|SetSpawnImmunityTime (select _output_get _spawnimmunitytime_legacy_2 _spawnimmunitytime))
            (Variables|Legacy|SetSpawnImmunityTime_Legacy _spawnimmunitytime_legacy)
            (Variables|SetbWasMigrated true)))
        (:then_1
          (bind _bmigrated_21 (Creative|Migration|MigrateOptionTypeBooleantoBoolean (Variables|Legacy|GetSpawnImmunityTime_Legacy_Override) _spawnimmunitytime_legacy_override (Variables|UserOptions|Overrides|GetSpawnImmunityTime_Override)))
          (if _bmigrated_21
            (Variables|UserOptions|Overrides|SetbOverrideSpawnImmunityTime_Override (Variables|UserOptions|Overrides|GetSpawnImmunityTime_Override))
            (Variables|SetbWasMigrated true)))
        (:then_2
          _))
      ; Migrate Grant Ammo With Weapons
      (if (!= _bgrantammowithweapons (Variables|Legacy|GetbGrantAmmoWithWeapons))
        (Variables|UserOptions|Overrides|SetSpareWeaponAmmo_Override (not (Variables|Legacy|GetbGrantAmmoWithWeapons)))
        (Variables|UserOptions|Player|Inventory|SetSpareWeaponAmmo 0)
        (Variables|Legacy|SetbGrantAmmoWithWeapons _bgrantammowithweapons)
        (Variables|SetbWasMigrated true)
        (else
          _))
      ; MigrateLocomotionSettings
      (bind _bmigrated_22 (Creative|Migration|MigrateOptionTypeBooleantoBoolean (Variables|Legacy|GetbApplyJumpPenalty) _bapplyjumppenalty (Variables|UserOptions|Player|Locomotion|GetbApplyJumpingPenalty)))
      (bind _bmigrated_23 (Creative|Migration|MigrateOptionTypeBooleantoBoolean (Variables|Legacy|GetbApplyJumpPenalty_Override) _bapplyjumppenalty_override (Variables|UserOptions|Overrides|GetbApplyJumpingPenalty_Override)))
      (bind _bmigrated_24 (Creative|Migration|MigrateOptionTypeBooleantoBoolean (Variables|Legacy|GetbGliderRedeploy) _bgliderredeploy (Variables|UserOptions|Player|Locomotion|GetbGliderRedeployable)))
      (bind _bmigrated_25 (Creative|Migration|MigrateOptionTypeBooleantoBoolean (Variables|Legacy|GetbGliderRedeploy_Override) _bgliderredeploy_override (Variables|UserOptions|Overrides|GetbGliderRedeployable_Override)))
      (bind _bmigrated_26 (Creative|Migration|MigrateOptionTypeBooleantoBoolean (Variables|Legacy|GetbFlyEnabled) _bflyenabled (Variables|UserOptions|Player|Locomotion|GetbFlyingEnabled)))
      (bind _bmigrated_27 (Creative|Migration|MigrateOptionTypeBooleantoBoolean (Variables|Legacy|GetbFlyEnabled_Override) _bflyenabled_override (Variables|UserOptions|Overrides|GetbFlyingEnabled_Override)))
      (Variables|SetbWasMigrated (Math|Boolean|ORBoolean _bmigrated_27 _bmigrated_26 _bmigrated_25 _bmigrated_24 _bmigrated_23 _bmigrated_22)))
    (:then_2
      (if (LocalVariables|GetbWasMigrated)
        (OptionData|RefreshPropertyOverrides (Variables|Default|GetToyOptionsComponent))))))

(fn RemoveClassSettingsFromMinigame
  (params (Minigame :type Fort%20Minigame%20Object%20Reference))
  (Utilities|FlowControl|Sequence
    (:then_0
      (bind _output_get (LocalVariables|SetSettingsMutator (ClassSettings|GetClassSettingsMutator (Variables|GetMinigame))))
      (Utilities|IsValid _output_get
        (:Is%20Valid
          (bind _returnvalue (ClassSettings|RemoveDataSource _output_get (Variables|Default|GetToyOptionsComponent)))
          (bind _returnvalue_2 (ClassSettings|RemoveDataSource (LocalVariables|GetSettingsMutator) self))
          (if (or _returnvalue_2 _returnvalue)
            (Minigame|GameRules|RefreshGameRulesPreview (Variables|GetMinigame))))))
    (:then_1
      (bind _output_get_2 (LocalVariables|SetWeaponSettingsMutator (ClassSettings|GetWeaponSettingsMutator (Variables|GetMinigame))))
      (Utilities|IsValid _output_get_2
        (:Is%20Valid
          (Class|FortAthenaMutatorWeaponSettings|RemoveDataSource _output_get_2 (Variables|Default|GetToyOptionsComponent)))))
    (:then_2
      (Class|FortAthenaMutatorUtilities|RemoveDataRegistryOptionsforContextObject))))

(fn RemoveItemsFromPlayer
  (params (Controller :type Controller%20Object%20Reference))
  (Utilities|IsValid Controller
    (:Is%20Valid
      (for _array_element (ItemData|MakeSpawnParamsList (Variables|Default|GetPickupItemList) (LocalVariables|GetSpawnTransform))
        (Inventory|RemoveItemsFromPlayerByIntStateValue (Variables|GetController) "PickupInstigatorHandle" (.PickupInstigatorHandle (Utilities|Struct|BreakMinigameSpawnerSpawnParams _array_element)) true)))))

(fn ShouldGrantedItems
  (params (UniqueID :type Unique%20Net%20Id%20Structure))
  (return (or (not (Utilities|Set|ContainsItem (Variables|Default|GetGrantedPlayers) UniqueID)) (Variables|UserOptions|Player|Inventory|GetbGrantItemsOnRespawn))))

(fn ShouldInvisiblePropsBeAllowed
  (return false))

(fn TryGiveAllItemsToPlayer
  (params (Controller :type Controller%20Object%20Reference))
  (bind _pickupitemlist (Variables|Default|GetPickupItemList))
  (bind _returnvalue (ItemData|MakeSpawnParamsList _pickupitemlist (struct :Location "0, 0, 0" :Rotation "0, 0, 0" :Scale "1.000000,1.000000,1.000000")))
  (bind _fortplayercontroller (LocalVariables|GetFortPlayerController))
  (bind _fortplayercontroller_2 (LocalVariables|GetFortPlayerController))
  (bind _inventoryowner (LocalVariables|GetInventoryOwner))
  (Fortnite|FortShippingLog self "Device_ClassDesigner_V2 TryGiveAllItemsToPlayer" true)
  (bind _asfort_player_controller (Utilities|Casting|CastToFortPlayerController Controller)
    (:then
      _)
    (:CastFailed
      _))
  (LocalVariables|SetFortPlayerController _asfort_player_controller)
  (Inventory|RemoveItemsfromPlayer Controller)
  (Utilities|FlowControl|Sequence
    (:then_0
      (if (Utilities|IsValid (LocalVariables|GetFortPlayerController) :!class /Script/Engine.KismetSystemLibrary)
        (Variables|/Script/FortniteGame.FortPlayerControllerAthena|SetbShouldForceDeleteDroppedItems true (LocalVariables|GetFortPlayerController))
        (else
          (Fortnite|FortShippingLog self "Device_ClassDesigner_V2 Unable to Grant Items. FortPlayerController is invalid." true))))
    (:then_1
      (bind _asfort_inventory_owner_interface (Utilities|Casting|CastToFortInventoryOwnerInterface (Variables|GetController))
        (:then
          (LocalVariables|SetInventoryOwner _asfort_inventory_owner_interface)
          (bind _output_get (LocalVariables|SetEquipItemIndexOnGrant (- (Math|Conversions|ToInteger(Byte) (Variables|UserOptions|Player|Inventory|GetEquipGrantedItem)) 1)))
          (for (_array_element _array_index) _returnvalue
            (bind _structout (Default|TryOverrideInitialWeaponAmmunition (Variables|Default|GetPickupItemList) _array_element (Variables|UserOptions|Overrides|GetInitialWeaponAmmo_Override) (Variables|UserOptions|Player|Inventory|GetInitialWeaponAmmo)))
            (Utilities|IsValid (Utilities|ResolveSoftReference (Utilities|Struct|BreakMinigameSpawnerSpawnParams _structout))
              (:Is%20Valid
                (Inventory|GiveItemtoInventoryOwner _inventoryowner (Utilities|ResolveSoftReference (Utilities|Struct|BreakMinigameSpawnerSpawnParams _structout)) (.ItemVariantGuid (Utilities|Struct|BreakMinigameSpawnerSpawnParams _structout)) (.PickupQuantity (Utilities|Struct|BreakMinigameSpawnerSpawnParams _structout)) true -1 (.PickupInstigatorHandle (Utilities|Struct|BreakMinigameSpawnerSpawnParams _structout)) true (.WeaponAmmoOverride (Utilities|Struct|BreakMinigameSpawnerSpawnParams _structout)))
                (Fortnite|FortShippingLog self "Device_ClassDesigner_V2 Granted item to player" true)
                (Inventory|MaybeGiveAmmo _inventoryowner (Utilities|ResolveSoftReference (Utilities|Struct|BreakMinigameSpawnerSpawnParams _structout)) (.PickupInstigatorHandle (Utilities|Struct|BreakMinigameSpawnerSpawnParams _structout)))
                (if (== (Math|Integer|Min(Integer) (Utilities|Array|LastIndex _returnvalue) _output_get) _array_index)
                  (if (Utilities|IsValid _fortplayercontroller :!class /Script/Engine.KismetSystemLibrary)
                    (Inventory|ExecuteInventoryItemDefinition :self _fortplayercontroller :ItemDefinition (Utilities|ResolveSoftReference (Utilities|Struct|BreakMinigameSpawnerSpawnParams _structout)) :Delay 0.0 :bForceExecute true :bActivateSlotAfterSettingFocused true)
                    (Fortnite|FortShippingLog self "Device_ClassDesigner_V2 Execute Inventory Item Definition" true))))
              (:Is%20Not%20Valid
                (Fortnite|FortShippingLog self "Device_ClassDesigner_V2 Unable to Grant Items. Pickup To Spawn is not valid." true)))))
        (:CastFailed
          (Fortnite|FortShippingLog self "Device_ClassDesigner_V2 Unable to Grant Items. Unable to cast to FortInventoryOwnerInterface." true))))
    (:then_2
      (if (Utilities|IsValid _fortplayercontroller_2 :!class /Script/Engine.KismetSystemLibrary)
        (Variables|/Script/FortniteGame.FortPlayerControllerAthena|SetbShouldForceDeleteDroppedItems false _fortplayercontroller_2)))))

(fn ConstructionScript
  (Networking|CanExecuteCosmeticEvents
    (:True
      (Variables|BuildingActor|SetOverridePrimitivesToExcludeFoliage (Utilities|Array|MakeArray (Variables|Default|GetBaseMesh))))))
