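; Creative Class Selector device (V2) that switches a player's class, team or both through a timed zone or received events.
; Switch attempts pass use restrictions, then change team before class to protect new loadouts, with optional forced respawn.
; Change events outside active play are queued for warmup or game start, and burst effects use an added pawn component.
; It also recolors its visuals by accent or team and migrates legacy options.
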(event OnPropertiesLoaded(ToyOptions)
  ;[0.0199 0.4896 0.1098 0.5] Initialization
  (CallFunction|InitializeDevice))

(event OnAnyPropertyChanged(ToyOptions)
  ;[0.0199 0.4896 0.1098 0.5] Initialization
  (CallFunction|InitializeDevice))

(event Custom|Initialize%20Device
  ;[0.0199 0.4896 0.1098 0.5] Initialization
  (Initialization|InitializeVisibilityComponent) ; Limitation on the Visibility Component means it currently needs called on world ready.
  (CallFunction|MigrateSameChannelInstances)
  (Initialization|InitializeComponents)
  (Effects|UpdateEnabledEffects)
  (Effects|UpdateDeviceColors)
  (Initialization|RefreshDelegateBindings true)
  (Initialization|SetDeviceTeam))

(event OnAddedtoMinigame(MinigameLogic)
  (params (Minigame :type Fort%20Minigame%20Object%20Reference))
  ;[0.0199 0.4896 0.1098 0.5] Initialization
  ;;[0.0199 0.4896 0.1098 0.5] Store a copy of the local minigame when we're added. We'll be using it a lot.
  (Variables|Default|SetCurrent%20Minigame Minigame)
  (Initialization|RefreshDelegateBindings true))

(event OnRemovedfromMinigame(MinigameLogic)
  (params (Minigame :type Fort%20Minigame%20Object%20Reference))
  ;[1 0 0.07061] Unbind on Removal
  (Utilities|IsValid Minigame
    (:Is%20Valid
      (Initialization|RefreshDelegateBindings false))))

(event OnParticipantActivatedZone(Creative_MeshZone_Component)
  (params (Controller :type Controller%20Object%20Reference))
  ;[0 0.6006 1 0.5] Zone interaction has completed, activate the class/team switch
  (bind _asfort_player_state (Utilities|Casting|CastToFortPlayerState (Variables|/Script/Engine.Controller|GetPlayerState Controller)))
  (PlayerSwitch|AttemptSwitchPlayer _asfort_player_state))

(event MinigameWarmupDelegate(MinigameLogic)
  (Utilities|FlowControl|Sequence
    (:then_0
      (bind _array_element (Utilities|CheckforDelayedEvents (Variables|Default|GetTeamChangedInstigators) (Variables|UserOptions|GetbSendGameStartTeamChangesAtWarmup))
        (:LoopBody
          (CallFunction|CallTeamChangedGenController _array_element))))
    (:then_1
      (bind _array_element_2 (Utilities|CheckforDelayedEvents (Variables|Default|GetClassChangedInstigators) (Variables|UserOptions|GetbSendGameStartClassChangesAtWarmup))
        (:LoopBody
          (Default|CallClassChangedGenController _array_element_2))))))

(event MinigameStartedDelegate(MinigameLogic)
  (Utilities|FlowControl|Sequence
    (:then_0
      (bind _array_element (Utilities|CheckforDelayedEvents (Variables|Default|GetTeamChangedInstigators) (Variables|UserOptions|GetbSendGameStartTeamChangesAtWarmup))
        (:LoopBody
          (CallFunction|CallTeamChangedGenController _array_element))))
    (:then_1
      (bind _array_element_2 (Utilities|CheckforDelayedEvents (Variables|Default|GetClassChangedInstigators) (Variables|UserOptions|GetbSendGameStartClassChangesAtWarmup))
        (:LoopBody
          (Default|CallClassChangedGenController _array_element_2))))))

(event Custom|OnClassChangeComplete
  (params (PlayerState :type Fort%20Player%20State%20Object%20Reference) (ClassSlotIndex :type Byte) (OldClassSlotIndex :type Byte))
  (bind _controller (CreativeAI|Controller|GetParticipantControllerfromActor PlayerState))
  ;[0 0.6006 1 0.5] When the Class or Team Change Completes, send the appropriate trigger message and multicast the effect playing on the affected pawn (audio + gameplay cue)
  (Utilities|IsValid PlayerState
    (:Is%20Valid
      (bind _bshouldcontinueexecution (CallFunction|DelayonMinigameState (Variables|Default|GetClassChangedInstigators) _controller))
      (if _bshouldcontinueexecution
        (Default|CallClassChangedGenController _controller)))))

(event HandleVersionUpgrades(CreativeVersion)
  (params (LastSavedVersion :type Creative%20Version%20Id%20Structure))
  (Utilities|FlowControl|Sequence
    (:then_0
      ;[1 0 0.6677 0.649] Actors placed before the RemoveThrottle tag need to keep the Throttle Class Change bool on (maintain old behavior)
      (CreativeVersion|CompareVersiontoTag LastSavedVersion "(TagName=\"Creative.Version.ClassSelector.RemoveThrottle\")"
        (:OlderVersion
          (Variables|UserOptions|SetThrottleClassChanges true)
          (Variables|Default|SetWasUpgraded true))))
    (:then_1
      ;[0.2632 1 0.9701] Expand ClearOnSwitch user option
      (CreativeVersion|CompareVersiontoTag LastSavedVersion "(TagName=\"Creative.Version.ClassSelector.ExpandClearOnSwitch\")"
        (:OlderVersion
          (Variables|UserOptions|SetClearItemOnSwitchRule (select (Variables|Legacy|GetClearItemsOnSwitch) "ClearOnClassOrTeamSwitch" "ClearOnClassSwitch"))
          (Variables|Default|SetWasUpgraded true))))
    (:then_2
      (if (Variables|Default|GetWasUpgraded)
        (OptionData|RefreshPropertyOverrides (Variables|Default|GetToyOptions))))))

(event Custom|ChangePlayerToClassFromEvent
  (params (Instigator :type Controller%20Object%20Reference))
  ;[0.3536 0 1 0.241] Transmitters / Receivers
  (Utilities|IsValid Instigator
    (:Is%20Valid
      (bind _asfort_player_state (Utilities|Casting|CastToFortPlayerState (Variables|/Script/Engine.Controller|GetPlayerState Instigator)))
      (PlayerSwitch|AttemptSwitchPlayer _asfort_player_state false))))

(event Custom|ChangePlayerToTeamFromEvent
  (params (Instigator :type Controller%20Object%20Reference))
  ;[0.3536 0 1 0.241] Transmitters / Receivers
  (Utilities|IsValid Instigator
    (:Is%20Valid
      (bind _asfort_player_state (Utilities|Casting|CastToFortPlayerState (Variables|/Script/Engine.Controller|GetPlayerState Instigator)))
      (PlayerSwitch|AttemptSwitchPlayer _asfort_player_state true false))))

(event Custom|ChangePlayerToTeamAndClassFromEvent
  (params (Instigator :type Controller%20Object%20Reference))
  ;[0.3536 0 1 0.241] Transmitters / Receivers
  (Utilities|IsValid Instigator
    (:Is%20Valid
      (bind _asfort_player_state (Utilities|Casting|CastToFortPlayerState (Variables|/Script/Engine.Controller|GetPlayerState Instigator)))
      (PlayerSwitch|AttemptSwitchPlayer _asfort_player_state))))

(event Custom|EnableFromEvent
  (params (Instigator :type Controller%20Object%20Reference))
  ;[0.3536 0 1 0.241] Transmitters / Receivers
  (Class|CreativeEnabledComponent|SetEnabled (Variables|Default|GetEnabledComponent) true))

(event EventPostFinishSpawning
  ;[0.3536 0 1 0.241] Transmitters / Receivers
  (CallFunction|MigrateSameChannelInstances)
  (|CallParentFunction))

(event Custom|DisableFromEvent
  (params (Instigator :type Controller%20Object%20Reference))
  ;[0.3536 0 1 0.241] Transmitters / Receivers
  (Class|CreativeEnabledComponent|SetEnabled (Variables|Default|GetEnabledComponent) false))

(event OnPlayerTeamSwitched(TeamSelectionComponent)
  (params (Player :type Controller%20Object%20Reference) (New%20Team :type Integer) (Old%20Team :type Integer))
  ;[0.3536 0 1 0.241] Transmitters / Receivers
  (if (Class|CreativeBlueprintLibrary|IsAllowingTestPlayeronCreativeDevices)
    (bind _asfort_athena_aibot_controller (Utilities|Casting|CastToFortAthenaAIBotController Player)
      (:then
        (bind _bshouldcontinueexecution (CallFunction|DelayonMinigameState (Variables|Default|GetTeamChangedInstigators) _asfort_athena_aibot_controller))
        (if _bshouldcontinueexecution
          (CallFunction|CallTeamChangedGenController _asfort_athena_aibot_controller)))
      (:CastFailed
        _))
    (else
      _))
  (bind _asfort_player_controller (Utilities|Casting|CastToFortPlayerController Player))
  (bind _bshouldcontinueexecution_2 (CallFunction|DelayonMinigameState (Variables|Default|GetTeamChangedInstigators) _asfort_player_controller))
  (if _bshouldcontinueexecution_2
    (CallFunction|CallTeamChangedGenController _asfort_player_controller)))

(event Custom|Update%20Team%20Event
  (params (Instigator :type Controller%20Object%20Reference))
  ;[0.3536 0 1 0.241] Transmitters / Receivers
  (bind _valid_controller (CreativeUserRestrictions|ActorHasValidParticipantController (Variables|Default|GetUseRestrictionComponent) Instigator true true true))
  (if _valid_controller
    (Initialization|InitializeComponent (Variables|Default|GetTeamSelectionComponent) (Variables|Default|GetMinigameLogic) (Knobs|Team|ReshiftTeamValue (+ (Math|Conversions|ToInteger(Byte) (Team|GetActorPvPTeamIndex Instigator)) 1)) (select (or (Utilities|Enum|Equal(Enum) (Variables|UserOptions|GetClearItemOnSwitchRule) "ClearOnTeamSwitch") (Utilities|Enum|Equal(Enum) (Variables|UserOptions|GetClearItemOnSwitchRule) "ClearOnClassOrTeamSwitch")) "NewEnumerator1" "NewEnumerator2") :!class /Game/Creative/Devices/Team_Selector/Components/Creative_TeamSelection_Component.Creative_TeamSelection_Component_C)))

(event OnParticipantEntered(Creative_MeshZone_Component)
  (params (Controller :type Controller%20Object%20Reference) (HitResult :type Hit%20Result%20Structure))
  (bind (_selector_state _team_index) (Getters|GetTeam (Variables|Default|GetTeamSelectionComponent)))
  ;[1 0.7441 0 0.5] If the player entering the zone will have an effect (will change either the player's class or team). The delegate is only called if the player can use the volume, so that's already covered.
  (bind _asfort_player_state (Utilities|Casting|CastToFortPlayerState (Variables|/Script/Engine.Controller|GetPlayerState Controller)))
  (if (or (not (ClassSettings|IsPlayerMatchingClass _asfort_player_state)) (!= (Math|Conversions|ToInteger(Byte) (Team|GetActorPvPTeamIndex Controller)) _team_index))
    (Effects|PulseMesh (Variables|Default|GetCreative_MeshZone_Component) HitResult)))

(event OnEnabledStateChanged(EnabledComponent)
  (params (Enabled :type Boolean))
  ;[1 0.7441 0 0.5] Enabling or Disabling the lights and audio when the Enabled State or Visibility Changes.
  (Effects|UpdateEnabledEffects))

(event OnVisibilityChanged(VolumeAreaMeshVisibleInGameComponent)
  (params (Visible :type Boolean))
  ;[1 0.7441 0 0.5] Enabling or Disabling the lights and audio when the Enabled State or Visibility Changes.
  (Effects|UpdateEnabledEffects))

(event Custom|OnTeamConfigChanged
  (Networking|CanExecuteCosmeticEvents
    (:True
      (Effects|UpdateDeviceColors))))

(event OnDeviceTeamUpdated(TeamSelectionComponent)
  (params (Selection%20State :type ECreativeTeamSelectorState%20Enum :default "NewEnumerator0") (Selected%20Team :type Integer))
  (Networking|CanExecuteCosmeticEvents
    (:True
      (Effects|UpdateDeviceColors))))

(fn ApplyPlayerFX
  (params (Pawn :type Pawn%20Object%20Reference))
  ;[1 0.7441 0 0.5] Play a Sound and an effect on the pawn when a player successfully switches class.
  ; This is done through a generic pawn component in order to allow this to function when the device is out of relevancy.
  ; This is needed instead of a GameplayCue as we need to pass the color, which I can't do via a gameplay cue.
  (Utilities|IsValid Pawn
    (:Is%20Valid
      (Networking|SwitchHasAuthority
        (:Authority
          (if (or (Variables|UserOptions|GetDisplayVFXEffectOnActivation) (Variables|UserOptions|Basic|GetActivationAudio))
            (bind _output_get (LocalVariables|SetFoundComponent (Actor|GetComponentbyClass Pawn "/CreativeCoreDevices/ClassSelector/FortPawnComponent_GenericBurstEffect.FortPawnComponent_GenericBurstEffect_C")))
            (Utilities|IsValid _output_get
              (:Is%20Valid
                _)
              (:Is%20Not%20Valid
                (bind _returnvalue (AddComponent|NODE_AddFortPawnComponent_GenericBurstEffect-0 :self Pawn))
                (LocalVariables|SetFoundComponent _returnvalue)))
            (Class|FortPawnComponentGenericBurstEffect|ApplyEffect (LocalVariables|GetFoundComponent) (select (Variables|UserOptions|GetDisplayVFXEffectOnActivation) (Variables|Default|GetActivationVFX) 0) (select (Variables|UserOptions|Basic|GetActivationAudio) (Variables|Audio|GetSwitch%20Sound) 0) (Default|GetTeamLinearColor))))))))

(fn AttemptSwitchPlayer
  (params (PlayerState :type Fort%20Player%20State%20Object%20Reference) (Switch%20Team :type Boolean :default true) (Switch%20Class :type Boolean :default true))
  (bind (_should_switch _class_id) (CallFunction|GetClassToSwitch))
  (Variables|Default|SetbPlayNextChangeEffect (or Switch%20Team Switch%20Class))
  (if (CreativeUserRestrictions|CanParticipantActorInteract (Variables|Default|GetUseRestrictionComponent) PlayerState true true false)
    (Utilities|IsValid PlayerState
      (:Is%20Valid
        (Fortnite|FortShippingLog self (Utilities|String|Append "[Attempt Switch Player Event] - " "Switch Teams [" (Utilities|String|ToString(Boolean) (Variables|GetSwitch%20Team)) "] - Switch Class [" (Utilities|String|ToString(Boolean) (Variables|GetSwitch%20Class)) "]"))
        (Utilities|FlowControl|Sequence
          (:then_0
            ;[1 1 1] Switch Team. This has to happen first as it currently always respawns the player. Don't want him being assigned class weapons and immediately dumping them on the ground.
            (if (and (Variables|GetSwitch%20Team) (Variables|UserOptions|Basic|GetTeamToSwitchTo_Override))
              (bind _bswitchsuccess (PlayerTeamSwitch|AttemptPlayerTeamSwitch (Variables|Default|GetTeamSelectionComponent) (CreativeAI|Controller|GetParticipantControllerfromActor (Variables|GetPlayerState)) (not (Variables|UserOptions|GetbRespawnPlayerOnSwitch)) (Variables|UserOptions|GetRestoreHealthAndShieldsOnSwitch) (select (or (Utilities|Enum|Equal(Enum) (Variables|UserOptions|GetClearItemOnSwitchRule) "ClearOnTeamSwitch") (Utilities|Enum|Equal(Enum) (Variables|UserOptions|GetClearItemOnSwitchRule) "ClearOnClassOrTeamSwitch")) true false)))
              (LocalVariables|SetSwitchSuccess _bswitchsuccess)))
          (:then_1
            ;[1 1 1] Switch Class
            (if (and (Variables|GetSwitch%20Class) _should_switch)
              (bind _current_minigame (Variables|Default|GetCurrent%20Minigame)
                (:then
                  (Utilities|IsValid (Minigame|GetFNEClassComponent _current_minigame)
                    (:Is%20Valid
                      (bind _returnvalue (Minigame|SetClassSlotforPlayer (Minigame|GetFNEClassComponent _current_minigame) (Variables|GetPlayerState) _class_id (EventDispatchers|CreateEvent :!event OnClassChangeComplete) false (Variables|UserOptions|GetRestoreHealthAndShieldsOnSwitch) (Variables|UserOptions|GetClearItemOnSwitchRule) 0.0 (Math|Float|SelectFloat 3.0 0.0 (Variables|UserOptions|GetThrottleClassChanges))))
                      (bind _output_get (LocalVariables|SetSwitchSuccess (or _returnvalue (LocalVariables|GetSwitchSuccess))))
                      (Fortnite|FortShippingLog self (Utilities|String|Append "[Attempt Switch Player Event] - Set Class Success [" (Utilities|String|ToString(Boolean) _output_get) "]")))))
                (:else))))
          (:then_2
            (if (and (Variables|UserOptions|GetbRespawnPlayerOnSwitch) (LocalVariables|GetSwitchSuccess))
              (bind _current_minigame_2 (Variables|Default|GetCurrent%20Minigame)
                (:then
                  (PlayerSpawn|ForceRespawn (Minigame|GetPlayerSpawnComponent _current_minigame_2) (Variables|GetPlayerState) (GameplayTags|MakeGameplayTagContainerfromTag "(TagName=\"Gameplay.Damage.MinigameClassSlotSwitchSuicide\")") :!class /Script/FortniteGame.FortMinigamePlayerSpawnComponent))
                (:else)))))))))

(fn CallClassChanged_GenController
  (params (Controller :type Controller%20Object%20Reference))
  (GameplayEvent|BroadcastEvent (Variables|UserOptions|Events|GetOnClassSwitched) 0 (Utilities|Struct|MakeGameplayEventPlayerClassChanged Controller))
  (Utilities|IsValid Controller
    (:Is%20Valid
      (Effects|PlayNextChangeEffectOnce (Pawn|GetControlledPawn Controller)))))

(fn CallTeamChanged_GenController
  (params (Controller :type Controller%20Object%20Reference))
  (GameplayEvent|BroadcastEvent (Variables|UserOptions|Events|GetOnTeamSwitched) 0 (Utilities|Struct|MakeGameplayEventPlayerTeamChanged Controller))
  (Utilities|IsValid Controller
    (:Is%20Valid
      (Effects|PlayNextChangeEffectOnce (Pawn|GetControlledPawn Controller)))))

(fn CreateOrAssignMIDs
  (Material|ReattachOrCreateMID (Variables|Default|GetBaseMesh) (Variables|Effects|GetClassSelectorBaseMID))
  (Variables|Effects|SetVolumeAreaMID (Material|GetMIDfromMaterial (Variables|Default|GetCreative_MeshZone_Component) 0 "/CreativeCoreDevices/CaptureArea/MI_CaptureArea_DirectColor.MI_CaptureArea_DirectColor"))
  (Variables|Effects|SetLightRaysMID (Material|GetMIDfromMaterial (Variables|Default|GetLightRays) 0 "/Game/Items/Traps/Materials/M_NoWeaponZone_Rays_Inst.M_NoWeaponZone_Rays_Inst")))

(fn DelayOnMinigameState
  (params (InstigatorsArray :type Array%20of%20Controller%20Object%20References) (Instigator :type Controller%20Object%20Reference))
  (bind _current_minigame (Variables|Default|GetCurrent%20Minigame)
    (:then
      (if (Utilities|Enum|Equal(Enum) (Variables|/Script/FortniteGame.FortMinigame|GetCurrentState _current_minigame) "InProgress")
        (return true)
        (else
          (Utilities|Array|Add (Variables|GetInstigatorsArray) (Variables|Actor|GetInstigator))
          (return false))))
    (:else
      (return false))))

(fn GetClasstoSwitch
  ;[1 1 1] -3 for Class Index is Do Not Switch. This is a helper function to minimise use of magic numbers (limitation of the playlistuseroption system for index lookup with special cases)
  (return (Variables|UserOptions|Basic|GetClassToSwitchTo_Override) (Creative|ClassOption|ConverttoClassSlotIndex (Variables|UserOptions|Basic|GetClassToSwitchTo))))

(fn GetTeamLinearColor
  (switch Utilities|FlowControl|Switch|SwitchonECreativeColorSetType (Variables|UserOptions|Basic|GetAccentColorType)
    (:NewEnumerator0
      (return 0))
    (:NewEnumerator1
      (return (Variables|UserOptions|Basic|GetAccentColor)))
    (:NewEnumerator2
      _)
    (:NewEnumerator3
      _))
  (return (Utilities|Array|Get(acopy) (Class|GetClassDefaults "/Game/Creative/Sets/MilitaryBase/Props/CreativeTeamColors.CreativeTeamColors_C") (Math|Integer|%25%28Integer%29 (Effects|GetAccentColorIndex) (Utilities|Array|Length (Class|GetClassDefaults "/Game/Creative/Sets/MilitaryBase/Props/CreativeTeamColors.CreativeTeamColors_C"))))))

(fn GetAccentColorIndex
  (bind (_selector_state _team_index) (Getters|GetTeam (Variables|Default|GetTeamSelectionComponent)))
  (bind (_selector_state_2 _team_index_2) (Getters|GetTeam (Variables|Default|GetTeamSelectionComponent)))
  (Utilities|FlowControl|Sequence
    (:then_0
      (switch Utilities|FlowControl|Switch|SwitchonECreativeColorSetType (Variables|UserOptions|Basic|GetAccentColorType)
        (:NewEnumerator1
          (bind (_found_color _color_index) (Creative|Color|GetCreativeTeamColorIndexFromColor (Variables|UserOptions|Basic|GetAccentColor)))
          (if _found_color
            (LocalVariables|SetLocalIndex _color_index)))
        (:NewEnumerator2
          (bind _currentminigame (Variables|/Script/FortniteGame.FortMinigameLogicComponent|GetCurrentMinigame (Variables|Default|GetMinigameLogic))
            (:then
              (Utilities|IsValid (Minigame|GetMinigameTeamComponent _currentminigame)
                (:Is%20Valid
                  (LocalVariables|SetLocalIndex (Utilities|ClampMinInteger (Minigame|GameRules|GetTeamColorIndex (Minigame|GetMinigameTeamComponent _currentminigame) (Teams|IndextoTeam (select (Utilities|Enum|Equal(Enum) _selector_state "NewEnumerator0") (Knobs|Team|GetTeamValueasIndex (Teams|TeamtoIndex (Creative|TeamOption|ConverttoTeam (Variables|UserOptions|GetActivatingTeam)))) _team_index))))))
                (:Is%20Not%20Valid
                  _)))
            (:else
              _))
          (LocalVariables|SetLocalIndex (+ (Teams|TeamtoIndex (Creative|TeamOption|ConverttoTeam (Variables|UserOptions|GetTeamToSwitchTo))) 1)))
        (:NewEnumerator3
          (Utilities|IsValid (Fortnite|GetFirstLocalFortPlayerController true)
            (:Is%20Valid
              (if (Variables|UserOptions|Basic|GetTeamToSwitchTo_Override)
                (if (Creative|TeamOption|IsTeamType (Variables|UserOptions|GetTeamToSwitchTo) "Random")
                  ;[1 0.7133 0 0.35] 0= Neutral
                  ; 1= Friendly
                  ; 2 = Hostile
                  (LocalVariables|SetLocalIndex 2) ; Random team will always switch,\nso is always hostile.
                  (else
                    _)))
              (LocalVariables|SetLocalIndex (Math|Integer|SelectInt 0 (Math|Integer|SelectInt 1 2 (== (Math|Byte|MakeLiteralByte (select (Utilities|Enum|Equal(Enum) _selector_state_2 "NewEnumerator0") (Creative|TeamOption|ConverttoTeam (Variables|UserOptions|GetActivatingTeam)) (Teams|IndextoTeam _team_index_2))) (Team|GetActorTeam (Fortnite|GetFirstLocalFortPlayerController true)))) (< (select (Utilities|Enum|Equal(Enum) _selector_state_2 "NewEnumerator0") (Creative|TeamOption|ConverttoTeam (Variables|UserOptions|GetActivatingTeam)) (Teams|IndextoTeam _team_index_2)) 3))))))))
    (:then_1
      (return (LocalVariables|GetLocalIndex)))))

(fn GetCreativeClassInventory
  (bind _classtoswitchto (Variables|UserOptions|Basic|GetClassToSwitchTo))
  (bind _returnvalue (Creative|ClassOption|ConverttoClassSlotIndex _classtoswitchto))
  (Utilities|Array|Clear (LocalVariables|GetInventory))
  (bind _current_minigame (Variables|Default|GetCurrent%20Minigame)
    (:then
      (Utilities|IsValid (Minigame|GetFNEClassComponent _current_minigame)
        (:Is%20Valid
          (if (.ReturnValue (Minigame|GameRules|GetClassSlotInventory (Minigame|GetFNEClassComponent _current_minigame) _returnvalue))
            (bind _output_get (LocalVariables|SetItemsAndNumberArray (Minigame|GameRules|GetClassSlotInventory (Minigame|GetFNEClassComponent _current_minigame) _returnvalue)))
            (for _array_element _output_get
              (bind (_array_element_count _array_element_item) (Utilities|Struct|BreakItemAndCount _array_element))
              (bind _asfort_world_item_definition (Utilities|Casting|CastToFortWorldItemDefinition _array_element_item))
              (Utilities|Array|Add (LocalVariables|GetInventory) _asfort_world_item_definition))
            (return (LocalVariables|GetInventory))
            (else
              _)))
        (:Is%20Not%20Valid
          _))
      (return (LocalVariables|GetInventory)))
    (:else
      _)))

(fn GetCreativeClassName
  (bind _current_minigame (Variables|Default|GetCurrent%20Minigame)
    (:then
      (Utilities|IsValid (Minigame|GetFNEClassComponent _current_minigame)
        (:Is%20Valid
          (return (select (or (Utilities|Text|TextIsEmpty (Minigame|GameRules|GetClassSlotName (Minigame|GetFNEClassComponent _current_minigame) (Creative|ClassOption|ConverttoClassSlotIndex (Variables|UserOptions|Basic|GetClassToSwitchTo)))) (not (Variables|UserOptions|Basic|GetClassToSwitchTo_Override))) "NSLOCTEXT(\"[2CFB85312ACB7C61D1F834AAF37C4838]\", \"86F9363F47FE09A7AF7B17866ECD300B\", \"No Class Slot Set\")" (Minigame|GameRules|GetClassSlotName (Minigame|GetFNEClassComponent _current_minigame) (Creative|ClassOption|ConverttoClassSlotIndex (Variables|UserOptions|Basic|GetClassToSwitchTo))))))
        (:Is%20Not%20Valid
          _)))
    (:else
      _))
  (return "NSLOCTEXT(\"[2CFB85312ACB7C61D1F834AAF37C4838]\", \"430A72D749DEBF49EBF84EA015C536EE\", \"No Class Info Available\")"))

(fn GetOverrideMeshMaterial
  (params (MeshComp :type Static%20Mesh%20Component%20Object%20Reference) (MatIdx :type Integer))
  (Utilities|FlowControl|Sequence
    (:then_0
      (if (Utilities|Equal(Object) (Variables|GetMeshComp) (Variables|Default|GetCreative_MeshZone_Component))
        (bind _volumeareamid (Variables|Effects|GetVolumeAreaMID)
          (:then
            (return _volumeareamid))
          (:else))))
    (:then_1
      (if (Utilities|Equal(Object) (Variables|GetMeshComp) (Variables|Default|GetBaseMesh))
        (bind _classselectorbasemid (Variables|Effects|GetClassSelectorBaseMID)
          (:then
            (return _classselectorbasemid))
          (:else))))
    (:then_2
      (if (Utilities|Equal(Object) (Variables|GetMeshComp) (Variables|Default|GetLightRays))
        (bind _lightraysmid (Variables|Effects|GetLightRaysMID)
          (:then
            (return _lightraysmid))
          (:else))))))

(fn InitializeComponents
  ;[1 1 1] Team Selector Component
  (Initialization|InitializeComponent (Variables|Default|GetTeamSelectionComponent) (Variables|Default|GetMinigameLogic) (Creative|TeamOption|GetLegacyTeamIndex (Variables|UserOptions|GetTeamToSwitchTo)) (select (or (Utilities|Enum|Equal(Enum) (Variables|UserOptions|GetClearItemOnSwitchRule) "ClearOnClassOrTeamSwitch") (Utilities|Enum|Equal(Enum) (Variables|UserOptions|GetClearItemOnSwitchRule) "ClearOnTeamSwitch")) "NewEnumerator1" "NewEnumerator2") :!class /Game/Creative/Devices/Team_Selector/Components/Creative_TeamSelection_Component.Creative_TeamSelection_Component_C)
  ;[1 1 1] Enabled Component
  (Class|CreativeEnabledComponent|Initialize (Variables|Default|GetEnabledComponent) (Variables|Default|GetMinigameLogic) (Variables|UserOptions|GetEnabledDuringPhase))
  ;[1 1 1] Visibility Component
  (Initialization|InitializeVisibilityComponent)
  ;[1 1 1] Mesh Zone Component
  (Initialization|InitializeComponent (Variables|Default|GetCreative_MeshZone_Component) (Variables|Default|GetEnabledComponent) (Variables|Default|GetMinigameLogic) (Variables|Default|GetUseRestrictionComponent) (Variables|UserOptions|Basic|GetSizeOfVolume) "1.000000,1.000000,0" (Variables|UserOptions|GetTimeToSwitch) (Variables|UserOptions|Basic|GetZoneAudio) (Variables|UserOptions|Basic|GetZoneActivationVFX) "NewEnumerator0" :!class /Game/Creative/Devices/Common/Components/Creative_MeshZone_Component.Creative_MeshZone_Component_C)
  ;[1 1 1] Use Restriction Component
  (Initialization|Initialize (Variables|Default|GetUseRestrictionComponent) (Variables|Default|GetMinigameLogic) (Variables|Default|GetEnabledComponent) (Variables|Default|GetToyOptions) (Variables|UserOptions|GetActivatingTeam) false (Utilities|Struct|MakeCreativeClassOption "Any") false false))

(fn InitializeVisibilityComponent
  (Initialization|InitializeWithSceneComponents (Variables|Default|GetBaseMeshVisibleInGameComponent) (Utilities|Array|MakeArray) (Utilities|Map|MakeMap (Variables|Default|GetBaseMesh) "QueryAndPhysics" (Variables|Default|GetLightRays) "NoCollision" (Variables|Default|GetLight) "NoCollision" (Variables|Default|GetAmbient%20Loop) "NoCollision") (select (Variables|UserOptions|Basic|GetVisibleDuringGame) "NewEnumerator3" "NewEnumerator1"))
  (Initialization|InitializeWithSceneComponents (Variables|Default|GetVolumeAreaMeshVisibleInGameComponent) (Utilities|Array|MakeArray (Variables|Default|GetCreative_MeshZone_Component)) (Utilities|Map|MakeMap) (select (Variables|UserOptions|GetVolumeVisibleInGame) "NewEnumerator3" "NewEnumerator1")))

(fn IsPlayerMatchingClass
  (params (PlayerState :type Fort%20Player%20State%20Object%20Reference))
  (Utilities|FlowControl|Sequence
    (:then_0
      (bind _current_minigame (Variables|Default|GetCurrent%20Minigame)
        (:then
          (Utilities|IsValid (Minigame|GetFNEClassComponent _current_minigame)
            (:Is%20Valid
              (LocalVariables|SetbSuccess (Math|Byte|Equal(Byte) (Minigame|GetClassSlotforPlayer (Minigame|GetFNEClassComponent _current_minigame) (Variables|GetPlayerState) true) (Creative|ClassOption|ConverttoClassSlotIndex (Variables|UserOptions|Basic|GetClassToSwitchTo)))))))
        (:else)))
    (:then_1
      (return (LocalVariables|GetbSuccess)))))

(fn MigrateLegacyUserOptions
  (bind (_classindex _teamindexsetting) (Class|GetClassDefaults "/CreativeCoreDevices/Device_ClassSelector_V2.Device_ClassSelector_V2_C"))
  (bind (_accentcolor _accentcolorindex) (Class|GetClassDefaults "/CreativeCoreDevices/Device_ClassSelector_V2.Device_ClassSelector_V2_C"))
  (bind _accentcolorindex_2 (Variables|Legacy|GetAccentColorIndex))
  (bind _teamindexsetting_2 (Variables|Legacy|GetTeamIndexSetting))
  (|CallParentFunction)
  (Utilities|FlowControl|Sequence
    (:then_0
      (Utilities|FlowControl|Sequence
        (:then_0
          ;[1 0 0.5777 0.25] AccentColorIndex -> AccentColor, AccentColorType
          (if (!= (Variables|Legacy|GetAccentColorIndex) _accentcolorindex)
            (if (>= _accentcolorindex_2 0)
              (Variables|UserOptions|Basic|SetAccentColor (Creative|Config|GetColoratTableIndex "/Game/Creative/Devices/Common/DataTables/DT_UserOption_ColorLibrary_TeamColors.DT_UserOption_ColorLibrary_TeamColors" _accentcolorindex_2))
              (else
                (Variables|UserOptions|Basic|SetAccentColor _accentcolor)))
            (Variables|UserOptions|Basic|SetAccentColorType (select (>= (Variables|Legacy|GetAccentColorIndex) 0) "NewEnumerator1" (select (== (Variables|Legacy|GetAccentColorIndex) -1) "NewEnumerator2" "NewEnumerator3")))
            (Variables|Legacy|SetAccentColorIndex _accentcolorindex)))
        (:then_1
          ;[0 0.5658 1 0.25] Migrating Team and Class
          (if (!= (Variables|Legacy|GetClassIndex) _classindex)
            (Variables|UserOptions|Basic|SetClassToSwitchTo_Override (!= (Variables|Legacy|GetClassIndex) -3))
            (Variables|UserOptions|Basic|SetClassToSwitchTo (Creative|ClassOption|MakefromLegacyClassIndex (Variables|Legacy|GetClassIndex)))
            (Variables|Legacy|SetClassIndex _classindex)
            (LocalVariables|SetOptionMigrated true)))
        (:then_2
          (if (!= _teamindexsetting (Variables|Legacy|GetTeamIndexSetting))
            (Variables|UserOptions|Basic|SetTeamToSwitchTo_Override (!= _teamindexsetting_2 -3))
            (Variables|UserOptions|SetTeamToSwitchTo (Creative|TeamOption|MakefromLegacyTeamIndex _teamindexsetting_2))
            (Variables|Legacy|SetTeamIndexSetting _teamindexsetting)
            (LocalVariables|SetOptionMigrated true)))))
    (:then_1
      (if (LocalVariables|GetOptionMigrated)
        (OptionData|RefreshPropertyOverrides (Variables|Default|GetToyOptions))))))

(fn MigrateSameChannelInstances
  (bind _bwasmigrated (UEFNConversionHelper|MigratetoReceiverwhenSameChannel (Variables|Default|GetChangeClassWhenReceived) (Variables|Default|GetChangeTeamWhenReceived) (Variables|Default|GetChangeTeamAndClassWhenReceived)))
  (return _bwasmigrated))

(fn PlayNextChangeEffectOnce
  (params (Pawn :type Pawn%20Object%20Reference))
  ;[1 1 1] Will call RPC PlayChangeEffect if the device wants to display the next effect And is set up to displayeffect
  (if (and (or (Variables|UserOptions|GetDisplayVFXEffectOnActivation) (Variables|UserOptions|Basic|GetActivationAudio)) (Variables|Default|GetbPlayNextChangeEffect))
    (CallFunction|ApplyPlayerFX Pawn)
    (Variables|Default|SetbPlayNextChangeEffect false)
    (return)
    (else
      (return))))

(fn RefreshDelegateBindings
  (params (Bind :type Boolean :default true))
  (bind _outputdelegate (EventDispatchers|CreateEvent :!event OnTeamConfigChanged))
  (bind _current_minigame (Variables|Default|GetCurrent%20Minigame)
    (:then
      (if (or Bind (and Bind (Utilities|Enum|NotEqual(Enum) (Variables|UserOptions|Basic|GetAccentColorType) "NewEnumerator1")))
        (Minigame|BindEventtoMinigameTeamConfigUpdated _current_minigame _outputdelegate)
        (else
          (Minigame|UnbindEventfromMinigameTeamConfigUpdated _current_minigame _outputdelegate))))
    (:else)))

(fn SetDeviceTeam
  ;[0 0.6006 1 0.5] Setting the Device Team. This essentially does nothing other than colors the outline red if you're not on the selected team.
  (CallFunction|SetTeam (Creative|TeamOption|ConverttoTeam (Variables|UserOptions|GetActivatingTeam))))

(fn UpdateDeviceColors
  (bind _new_color (Default|GetTeamLinearColor))
  (bind _returnvalue (Math|Color|LinearColor*Float _new_color 2.0))
  (bind _lightraysmid (Variables|Effects|GetLightRaysMID))
  (Networking|CanExecuteCosmeticEvents
    (:True
      (Effects|CreateorAssignMIDs)
      (Rendering|Material|SetScalarParameterValue (Variables|Effects|GetClassSelectorBaseMID) "TeamColor" (Math|Conversions|ToFloat(Integer) (Effects|GetAccentColorIndex)) :!class /Script/Engine.MaterialInstanceDynamic)
      (Rendering|Material|SetVectorParameterValue _lightraysmid "ColorHigh" _returnvalue :!class /Script/Engine.MaterialInstanceDynamic)
      (Rendering|Material|SetVectorParameterValue _lightraysmid "ColorLow" _returnvalue :!class /Script/Engine.MaterialInstanceDynamic)
      (Rendering|Components|Light|SetLightColor (Variables|Default|GetLight) _new_color false)
      (Rendering|Material|SetVectorParameterValue (Variables|Effects|GetVolumeAreaMID) "Color" _new_color :!class /Script/Engine.MaterialInstanceDynamic))))

(fn UpdateEnabledEffects
  (bind _volumevisibleingame (Variables|UserOptions|GetVolumeVisibleInGame))
  (bind _returnvalue (not _volumevisibleingame))
  (bind _ambient_loop (Variables|Default|GetAmbient%20Loop))
  (bind _ambient_loop_2 (Variables|Default|GetAmbient%20Loop))
  (bind _ambient_loop_3 (Variables|Default|GetAmbient%20Loop))
  ;[1 1 1] Switches the Light, Light Rays and Ambient Audio based on whether the Visible During Game user option is set to true or false.
  (Networking|CanExecuteCosmeticEvents
    (:True
      (UEFNConversionHelper|SetComponentHiddenEditor (Variables|Default|GetLightRays) _returnvalue)
      (UEFNConversionHelper|SetComponentHiddenEditor (Variables|Default|GetLight) _returnvalue)
      (if (Variables|UserOptions|Basic|GetZoneAudio)
        (Audio|Components|Audio|SetVolumeMultiplier (Variables|Default|GetAmbient%20Loop) 1.0 :!class /Script/Engine.AudioComponent)
        (if (Rendering|IsVisible (Variables|Default|GetAmbient%20Loop))
          (if (not (Audio|Components|Audio|IsPlaying _ambient_loop_2))
            (Audio|Components|Audio|FadeIn _ambient_loop_2 0.5 1.0 0.0 "Linear" :!class /Script/Engine.AudioComponent))
          (else
            (if (Audio|Components|Audio|IsPlaying _ambient_loop)
              (Audio|Components|Audio|FadeOut _ambient_loop 0.5 0.0 "Linear" :!class /Script/Engine.AudioComponent))))
        (else
          (Audio|Components|Audio|Stop _ambient_loop_3)
          (Audio|Components|Audio|SetVolumeMultiplier _ambient_loop_3 0.0 :!class /Script/Engine.AudioComponent))))))

(fn ConstructionScript
  ;[0.3237 1 0.2637] Update the visuals so that the light pillar will have the proper material when placed in a UEFN level in edit mode.
  (Effects|UpdateDeviceColors))
