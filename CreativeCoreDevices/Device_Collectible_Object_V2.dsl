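; Creative collectible object device that players pick up during a minigame, limited by team, class and consumption rules.
; An options-selected or custom mesh loads asynchronously, with optional pickup coloring.
; Authority handles collider overlaps, records collectors, sends trigger and stat events, then updates per-player visibility through replicated hide lists.
; Minigame, player and live-edit events refresh that state, device events drive respawn and visibility, and upgrade paths migrate older options.
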(event OnPropertiesLoaded(ToyOptionsComponent)
  (bind (_mesh _scale _stattag) (Utilities|Struct|BreakCreativeDeviceMeshSettings (Utilities|Array|Get(aref) (Variables|Default|GetCoinMeshScaleTagArray) (Math|Conversions|ToInteger(Byte) (CallFunction|GetCollectibleObject)))))
  (bind _newparam (CallFunction|GetCollectibleObject))
  (bind _returnvalue (Math|Conversions|ToInteger(Byte) _newparam))
  (bind _coinmeshscaletagarray (Variables|Default|GetCoinMeshScaleTagArray))
  ;[1 1 1] Updating tag if property changed / loaded
  (Class|CreativeUIBroadcastComponent|Initialize (Variables|Default|GetCreative_HUDBroadcast_Component) (Variables|Default|GetFortMinigameProgress) (Variables|UserOptions|Advanced|GetDisplayScoreUpdateOnHUD))
  (CallFunction|InitializeVisibilityComponent)
  ;;[1 0.2969 0.2969] Fix for FORT-477756. On Minigame Start is getting called before properties are load when loading directling into a match via an island code.
  (CallFunction|PickMesh)
  (if (and (Class|CreativeCommonDeviceFunctionLibrary|IsinPlayMode self) (Variables|Default|GetStartGameDelegateFired))
    (CallFunction|CollectibleInitialized)
    (else
      _))
  (if (Utilities|Array|IsValidIndex _coinmeshscaletagarray _returnvalue)
    (Networking|SwitchHasAuthority
      (:Authority
        (Variables|Default|SetCurrentCoinTag _stattag)
        (Initialization|InitializeComponentWithAffects (Variables|Default|GetCreative_UseRestriction_Component) (Variables|Default|GetFortMinigameProgress) 0 0 996 (Creative|ClassOption|GetLegacyClassIndex (Variables|UserOptions|GetAllowedClass)) 0 1 false false false)))))

(event OnAnyPropertyChanged(ToyOptionsComponent)
  (bind (_mesh _scale _stattag) (Utilities|Struct|BreakCreativeDeviceMeshSettings (Utilities|Array|Get(aref) (Variables|Default|GetCoinMeshScaleTagArray) (Math|Conversions|ToInteger(Byte) (CallFunction|GetCollectibleObject)))))
  (bind _newparam (CallFunction|GetCollectibleObject))
  (bind _returnvalue (Math|Conversions|ToInteger(Byte) _newparam))
  (bind _coinmeshscaletagarray (Variables|Default|GetCoinMeshScaleTagArray))
  ;[1 1 1] Updating tag if property changed / loaded
  (Class|CreativeUIBroadcastComponent|Initialize (Variables|Default|GetCreative_HUDBroadcast_Component) (Variables|Default|GetFortMinigameProgress) (Variables|UserOptions|Advanced|GetDisplayScoreUpdateOnHUD))
  (CallFunction|InitializeVisibilityComponent)
  ;;[1 0.2969 0.2969] Fix for FORT-477756. On Minigame Start is getting called before properties are load when loading directling into a match via an island code.
  (CallFunction|PickMesh)
  (if (and (Class|CreativeCommonDeviceFunctionLibrary|IsinPlayMode self) (Variables|Default|GetStartGameDelegateFired))
    (CallFunction|CollectibleInitialized)
    (else
      _))
  (if (Utilities|Array|IsValidIndex _coinmeshscaletagarray _returnvalue)
    (Networking|SwitchHasAuthority
      (:Authority
        (Variables|Default|SetCurrentCoinTag _stattag)
        (Initialization|InitializeComponentWithAffects (Variables|Default|GetCreative_UseRestriction_Component) (Variables|Default|GetFortMinigameProgress) 0 0 996 (Creative|ClassOption|GetLegacyClassIndex (Variables|UserOptions|GetAllowedClass)) 0 1 false false false)))))

(event EventBeginPlay
  (bind (_min _max) (Components|StaticMesh|GetLocalBounds (Variables|Default|GetVisibleMesh)))
  (bind _idlefx (Variables|Default|GetIdleFX))
  (bind _array (Utilities|Array|MakeArray _idlefx))
  (bind _collider (Variables|Default|GetCollider))
  (bind _array_2 (Utilities|Array|MakeArray _idlefx _collider))
  (bind _manually_set_collider (Variables|UserOptions|GetManually%20Set%20Collider))
  (bind _returnvalue (select _manually_set_collider _array _array_2))
  ;[1 1 1] Begin Play
  (Utilities|FlowControl|Sequence
    (:then_0
      ;[1 1 1] Begin Play Subscription - Team Switch
      (Networking|CanExecuteCosmeticEvents
        (:True
          (bind _asfort_player_controller_athena (Utilities|Casting|CastToFortPlayerControllerAthena (Game|GetPlayerController 0)))
          (FortPlayerControllerAthena|BindEventtoOnLocalPlayerChangedTeams _asfort_player_controller_athena (EventDispatchers|CreateEvent :!event Custom|UpdatePickupColorForPlayer)))))
    (:then_1
      ;[1 1 1] Mesh Setup
      (CallFunction|PickMesh)
      (Minigame|AddMinigameTag (Variables|Default|GetFortMinigameProgress) (Variables|Default|GetCurrentCoinTag))
      (Utilities|FlowControl|Sequence
        (:then_0
          ;[1 1 1] Setup Mesh Transform
          (if (Variables|Default|GetbApplyMeshTransform)
            (Transformation|SetWorldLocation _returnvalue (+ (Transformation|GetWorldLocation (Variables|Default|GetVisibleMesh)) (/ (+ _min _max) 2.0)) false false)
            (Transformation|SetWorldScale3D _returnvalue (/ (Transformation|GetWorldScale (Variables|Debug|GetStaticMeshComponent)) 2.0))))
        (:then_1
          ;[1 1 1] Collision & Visibility
          (Networking|CanExecuteCosmeticEvents
            (:True
              (Collision|SetCollisionResponsetoChannel (Variables|Default|GetCollider) "ECC_GameTraceChannel1" "ECR_Block")
              (Rendering|SetVisibility (Variables|Default|GetCollider) (Variables|UserOptions|GetManually%20Set%20Collider) false))))) ; I don't understand \nwhy it has to call\npart of this logic twice\nin a row???
    )))

(event Custom|Pick%20Mesh
  (bind _newparam (CallFunction|GetCollectibleObject))
  (bind (_min _max) (Components|StaticMesh|GetLocalBounds (Variables|Default|GetVisibleMesh)))
  (bind _returnvalue (Math|Conversions|ToInteger(Byte) _newparam))
  (bind _idlefx (Variables|Default|GetIdleFX))
  (bind _array (Utilities|Array|MakeArray _idlefx))
  (bind _collider (Variables|Default|GetCollider))
  (bind _array_2 (Utilities|Array|MakeArray _idlefx _collider))
  (bind _manually_set_collider (Variables|UserOptions|GetManually%20Set%20Collider))
  (bind _returnvalue_2 (select _manually_set_collider _array _array_2))
  (bind _coinmeshscaletagarray (Variables|Default|GetCoinMeshScaleTagArray))
  (bind (_currentmeshscaletag_mesh _currentmeshscaletag_scale _currentmeshscaletag_stattag _currentmeshscaletag_material) (Utilities|Struct|BreakCreativeDeviceMeshSettings (Variables|GetCurrentMeshScaleTag_Mesh)))
  (bind (_currentmeshscaletag_mesh_2 _currentmeshscaletag_scale_2) (Utilities|Struct|BreakCreativeDeviceMeshSettings (Variables|GetCurrentMeshScaleTag_Mesh)))
  (bind (_currentmeshscaletag_mesh_3 _currentmeshscaletag_scale_3 _currentmeshscaletag_stattag_2) (Utilities|Struct|BreakCreativeDeviceMeshSettings (Variables|GetCurrentMeshScaleTag_Mesh)))
  ;[1 1 1] Pick Mesh
  (if (Utilities|Array|IsValidIndex _coinmeshscaletagarray _returnvalue)
    (bind _output_get (Variables|Default|SetCurrentMeshScaleTag (Utilities|Array|Get(aref) _coinmeshscaletagarray _returnvalue)))
    (Variables|Default|SetCurrentCoinTag _currentmeshscaletag_stattag_2)
    (bind _loaded_object (Creative|Loading|AsyncLoadEditorSafe (select (and (Utilities|IsValidSoftObjectReference (Variables|UserOptions|GetCustomMesh)) (Variables|UserOptions|GetCustomMesh_Override)) (Variables|UserOptions|GetCustomMesh) (Utilities|Struct|BreakCreativeDeviceMeshSettings _output_get)))
      (:Loaded
        (bind _asstatic_mesh (Utilities|Casting|CastToStaticMesh _loaded_object))
        (Components|StaticMesh|SetStaticMesh (Variables|Default|GetVisibleMesh) _asstatic_mesh)
        (Collision|SetCollisionEnabled (Variables|Debug|GetStaticMeshComponent) "NoCollision")
        (Utilities|FlowControl|Sequence
          (:then_0
            ;[1 1 1] Setup Mesh Transform
            (Transformation|SetRelativeScale3D (Variables|Default|GetVisibleMesh) _currentmeshscaletag_scale_2)
            (if (Variables|Default|GetbApplyMeshTransform)
              (Transformation|SetWorldLocation _returnvalue_2 (+ (Transformation|GetWorldLocation (Variables|Default|GetVisibleMesh)) (/ (+ _min _max) 2.0)) false false)
              (Transformation|SetWorldScale3D _returnvalue_2 (/ (Transformation|GetWorldScale (Variables|Debug|GetStaticMeshComponent)) 2.0))))
          (:then_1
            ;[1 1 1] Collision & Visibility
            (Networking|CanExecuteCosmeticEvents
              (:True
                (Collision|SetCollisionResponsetoChannel (Variables|Default|GetCollider) "ECC_GameTraceChannel1" "ECR_Block")
                (Rendering|SetVisibility (Variables|Default|GetCollider) (Variables|UserOptions|GetManually%20Set%20Collider) false))))
          (:then_2
            ;[1 0.7441 0 0.5] Load Material once mesh has loaded
            (if (and (Utilities|IsValidSoftObjectReference _currentmeshscaletag_material) (not (and (Utilities|IsValidSoftObjectReference (Variables|UserOptions|GetCustomMesh)) (Variables|UserOptions|GetCustomMesh_Override))))
              (bind _loaded_object_2 (Creative|Loading|AsyncLoadEditorSafe _currentmeshscaletag_material)
                (:Loaded
                  (bind _asmaterial_interface (Utilities|Casting|CastToMaterialInterface _loaded_object_2)
                    (:then
                      (Rendering|Material|SetMaterial (Variables|Default|GetVisibleMesh) 0 _asmaterial_interface))
                    (:CastFailed
                      (Rendering|Material|SetMaterial (Variables|Default|GetVisibleMesh) 0 0)))
                  (Rendering|Material|SetMaterial (Variables|Default|GetVisibleMesh) 1 0) ; Always set mat 1 back to mesh default!
                  (Variables|Default|SetCurrentMeshComponentMaterial (Rendering|Material|GetMaterial (Variables|Default|GetVisibleMesh) 0))
                  ;[1 1 1] Pickup Color
                  (Networking|CanExecuteCosmeticEvents
                    (:True
                      (switch Utilities|FlowControl|Switch|SwitchonECollectibleObjectColorOption (Variables|UserOptions|GetCollectibleColor)
                        (:NewEnumerator0
                          _)
                        (:NewEnumerator1
                          (Utilities|FlowControl|Sequence
                            (:then_0
                              _)
                            (:then_1
                              (CallFunction|PickupColorSetupMID)
                              (bind _color (Default|PickupColorGetColor))
                              (CallFunction|PickupColorApplyColor _color))))
                        (:NewEnumerator2
                          _)
                        (:NewEnumerator3
                          _)
                        (:NewEnumerator4
                          _)
                        (:NewEnumerator5
                          _))
                      (Utilities|FlowControl|DoOnce true
                        (:Completed
                          (CallFunction|PickupColorUpdateIdleFXColor true))) ; We only need to reset the VFX if \nwe previously tweaked it using\none of the coloration options.
                    ))))
              (else
                _))))))))

(event Custom|UpdatePickupColorForPlayer
  ;[1 1 1] Pickup Color
  (bind _color (Default|PickupColorGetColor))
  (CallFunction|PickupColorApplyColor _color))

(event OnComponentBeginOverlap(Collider)
  (params (OverlappedComponent :type Primitive%20Component%20Object%20Reference) (OtherActor :type Actor%20Object%20Reference) (OtherComp :type Primitive%20Component%20Object%20Reference) (OtherBodyIndex :type Integer) (bFromSweep :type Boolean) (SweepResult :type Hit%20Result%20Structure%20%28by%20ref%29))
  ;[1 1 1] On Component Begin Overlap
  (Networking|SwitchHasAuthority
    (:Authority
      (bind _currentminigame (Variables|/Script/FortniteGame.FortMinigameLogicComponent|GetCurrentMinigame (Variables|Default|GetFortMinigameProgress))
        (:then
          (Minigame|IsInProgress _currentminigame
            (:Yes
              (if (Visibility|ShouldbeVisible (Variables|Default|GetVisibleInGameComponent))
                (bind _showfx (CallFunction|DoCollection OtherActor))
                (if _showfx
                  (CallFunction|ShowFX))))))
        (:else)))))

(event EventPostFinishSpawning
  ;[1 1 1] Turn off collision all day, everyday, no matter how you came into existence
  (|CallParentFunction)
  (CallFunction|MeshPickerUpgradePath)
  (CallFunction|OnlyVisibletoCollectingTeamUpgradePath)
  (Networking|SwitchHasAuthority
    (:Authority
      (Collision|SetCollisionEnabled (Variables|Debug|GetStaticMeshComponent) "NoCollision"))))

(event Custom|On%20Player%20Joined%20In%20Progress
  (params (Pawn :type Pawn%20Object%20Reference))
  ;[1 1 1] Server Update
  (Variables|Default|SetbRequiresServerUpdate true)
  (Utilities|FlowControl|DelayUntilNextTick)
  (CallFunction|ServerUpdateforPlayers))

(event Custom|On%20Player%20Added
  (params (UniqueNetId :type Unique%20Net%20Id%20Structure) (bIsLocalPlayer :type Boolean))
  ;[1 1 1] Server Update
  (Variables|Default|SetbRequiresServerUpdate true)
  (Utilities|FlowControl|DelayUntilNextTick)
  (CallFunction|ServerUpdateforPlayers))

(event Building|EventOnBuildingActorInitialized
  (params (InitializationReason :type EFortBuildingInitializationReason%20Enum :default "StaticallyPlaced") (BuildingPersistentState :type EFortBuildingPersistentState%20Enum :default "Default"))
  ;[1 1 1] Turn off collision all day, everyday, no matter how you came into existence
  (Networking|SwitchHasAuthority
    (:Authority
      (Collision|SetCollisionEnabled (Variables|Debug|GetStaticMeshComponent) "NoCollision"))))

(event LiveEdit|Object|EventOnPostPropertiesLiveEdited
  (params (ChangedProperties :type Array%20of%20Names))
  ;[1 1 1] Server Update
  (Variables|Default|SetbRequiresServerUpdate true)
  (Utilities|FlowControl|DelayUntilNextTick)
  (CallFunction|ServerUpdateforPlayers))

(event Custom|On%20Player%20Removed
  (params (UniqueNetId :type Unique%20Net%20Id%20Structure) (bIsLocalPlayer :type Boolean))
  ;[1 1 1] Server Update
  (Utilities|RemoveCollectorfromArray (Class|CreativeCommonDeviceFunctionLibrary|GetPlayerStateFromNetID UniqueNetId))
  (Utilities|FlowControl|DelayUntilNextTick)
  (CallFunction|ServerUpdateforPlayers))

(event Custom|Server%20Increment%20Pickup%20Flash
  ;[1 0 0.007519] This is only here to make the array 'HideForPlayers' replicate properly. This should be looked into
  (Variables|Default|SetPickupFlash (Math|Byte|Max(Byte) (Math|Byte|Byte+Byte (Variables|Default|GetPickupFlash) 1) 1)))

(event Custom|On%20Player%20Class%20Changed
  (params (PlayerState :type Fort%20Player%20State%20Object%20Reference) (ClassSlotIndex :type Byte))
  ;[1 1 1] Server Update
  (Utilities|FlowControl|DelayUntilNextTick)
  (CallFunction|UpdateIndividualPlayerwithNotifyonChange PlayerState))

(event Custom|Show%20FX
  ;[1 0.5253 0 0.2] Get sound and VFX to play on all clients
  (if (Math|Boolean|NORBoolean (Utilities|Enum|Equal(Enum) (Variables|UserOptions|GetShowPickupEffects) "NewEnumerator1") (Utilities|Enum|Equal(Enum) (Variables|UserOptions|GetShowPickupEffects) "NewEnumerator0"))
    (Networking|CanExecuteCosmeticEvents
      (:True
        (switch Utilities|FlowControl|Switch|SwitchonECreativeShowFX (Variables|UserOptions|GetShowPickupEffects)
          (:NewEnumerator2
            _)
          (:NewEnumerator3
            (if (Utilities|IsValidSoftObjectReference (Variables|UserOptions|GetCustomOnCollectedVFX))
              (Utilities|IsValid (Variables|Default|GetCachedCollectionPickupVFX)
                (:Is%20Valid
                  _)
                (:Is%20Not%20Valid
                  (bind _object (Utilities|AsyncLoadAsset (Variables|UserOptions|GetCustomOnCollectedVFX))
                    (:Completed
                      (bind _asniagara_system (Utilities|Casting|CastToNiagaraSystem _object))
                      (Variables|Default|SetCachedCollectionPickupVFX _asniagara_system)))))
              (Niagara|SpawnSystematLocation (Variables|Default|GetCachedCollectionPickupVFX) (Transformation|GetWorldLocation (Variables|Default|GetCollider)))
              (else
                (Effects|SpawnEmitteratLocation "/Game/Athena/Playlists/Score/Coins/FX/P_Athena_ScoreCoin_Collect_Gold.P_Athena_ScoreCoin_Collect_Gold" (Transformation|GetWorldLocation (Variables|Default|GetCollider))))))
          (:NewEnumerator4
            (Utilities|FlowControl|Sequence
              (:then_0
                _)
              (:then_1
                _))))
        (Audio|PlaySoundatLocation (Variables|UserOptions|Advanced|GetScoreCoinAudioQueue) (Transformation|GetWorldLocation (Variables|Default|GetCollider)))))))

(event MinigameStartedDelegate(FortMinigameProgress)
  (bind _minigame (Utilities|GetMinigame))
  ;[1 1 1] On Minigame Started
  (Utilities|FlowControl|Sequence
    (:then_0
      (Networking|SwitchHasAuthority
        (:Authority
          (CallFunction|ServerUpdateforPlayers))))
    (:then_1
      (if (Variables|Debug|GetbSelfPickUp)
        (Utilities|Time|SetTimerbyEvent (delegate Custom|Timer) 15.0)))
    (:then_2
      (Minigame|BindEventtoPlayerAddedDelegate _minigame (EventDispatchers|CreateEvent :!event Custom|On%20Player%20Added))
      (Minigame|BindEventtoPlayerRemovedDelegate _minigame (EventDispatchers|CreateEvent :!event Custom|On%20Player%20Removed))
      (Minigame|BindEventtoOnPlayerClassApplied _minigame (EventDispatchers|CreateEvent :!event Custom|On%20Player%20Class%20Changed))
      (Minigame|BindEventtoOnPlayerClassRemoved _minigame (EventDispatchers|CreateEvent :!event Custom|On%20Player%20Class%20Changed))
      (Minigame|BindEventtoOnPlayerJoininProgress _minigame (EventDispatchers|CreateEvent :!event Custom|On%20Player%20Joined%20In%20Progress))
      (Minigame|AssignPlayerChangedTeamDelegate _minigame (EventDispatchers|CreateEvent :!event PlayerChangedTeam)))
    (:then_3
      (Collision|SetCollisionResponsetoChannel (Variables|Default|GetCollider) "ECC_GameTraceChannel1" "ECR_Ignore"))))

(event MinigameStateChangedDelegate(FortMinigameProgress)
  (params (Minigame :type Fort%20Minigame%20Object%20Reference) (NewMinigameState :type EFortMinigameState%20Enum :default "PreGame"))
  (bind _minigame (Utilities|GetMinigame))
  ;[1 1 1] On Minigame Started
  (if (Utilities|Enum|Equal(Enum) NewMinigameState "Warmup")
    (CallFunction|UpdateVisibilityonGameStart)
    (Utilities|FlowControl|Sequence
      (:then_0
        (Networking|SwitchHasAuthority
          (:Authority
            (CallFunction|ServerUpdateforPlayers))))
      (:then_1
        (if (Variables|Debug|GetbSelfPickUp)
          (Utilities|Time|SetTimerbyEvent (delegate Custom|Timer) 15.0)))
      (:then_2
        (Minigame|BindEventtoPlayerAddedDelegate _minigame (EventDispatchers|CreateEvent :!event Custom|On%20Player%20Added))
        (Minigame|BindEventtoPlayerRemovedDelegate _minigame (EventDispatchers|CreateEvent :!event Custom|On%20Player%20Removed))
        (Minigame|BindEventtoOnPlayerClassApplied _minigame (EventDispatchers|CreateEvent :!event Custom|On%20Player%20Class%20Changed))
        (Minigame|BindEventtoOnPlayerClassRemoved _minigame (EventDispatchers|CreateEvent :!event Custom|On%20Player%20Class%20Changed))
        (Minigame|BindEventtoOnPlayerJoininProgress _minigame (EventDispatchers|CreateEvent :!event Custom|On%20Player%20Joined%20In%20Progress))
        (Minigame|AssignPlayerChangedTeamDelegate _minigame (EventDispatchers|CreateEvent :!event PlayerChangedTeam)))
      (:then_3
        (Collision|SetCollisionResponsetoChannel (Variables|Default|GetCollider) "ECC_GameTraceChannel1" "ECR_Ignore")))))

(event Custom|Timer
  ;[1 1 1] On Minigame Started
  (bind _showfx (CallFunction|DoCollection self))
  (if _showfx
    (CallFunction|ShowFX)))

(event Custom|Turn%20Visibility%20On%20Event
  (params (Instigator :type Controller%20Object%20Reference))
  ;[1 1 1] Device Events
  (Networking|SwitchHasAuthority
    (:Authority
      (Visibility|SetForcedVisibilityDuringGame (Variables|Default|GetVisibleInGameComponent) "NewEnumerator1"))))

(event Custom|CollectibleInitialized
  (bind _minigame (Utilities|GetMinigame))
  ;[1 1 1] On Minigame Started
  ;;[1 1 1] Make sure we update the collectible for all the players if they are initialized before device properties are loaded
  (Variables|Default|SetbRequiresServerUpdate true)
  (Utilities|FlowControl|Sequence
    (:then_0
      (Networking|SwitchHasAuthority
        (:Authority
          (CallFunction|ServerUpdateforPlayers))))
    (:then_1
      (if (Variables|Debug|GetbSelfPickUp)
        (Utilities|Time|SetTimerbyEvent (delegate Custom|Timer) 15.0)))
    (:then_2
      (Minigame|BindEventtoPlayerAddedDelegate _minigame (EventDispatchers|CreateEvent :!event Custom|On%20Player%20Added))
      (Minigame|BindEventtoPlayerRemovedDelegate _minigame (EventDispatchers|CreateEvent :!event Custom|On%20Player%20Removed))
      (Minigame|BindEventtoOnPlayerClassApplied _minigame (EventDispatchers|CreateEvent :!event Custom|On%20Player%20Class%20Changed))
      (Minigame|BindEventtoOnPlayerClassRemoved _minigame (EventDispatchers|CreateEvent :!event Custom|On%20Player%20Class%20Changed))
      (Minigame|BindEventtoOnPlayerJoininProgress _minigame (EventDispatchers|CreateEvent :!event Custom|On%20Player%20Joined%20In%20Progress))
      (Minigame|AssignPlayerChangedTeamDelegate _minigame (EventDispatchers|CreateEvent :!event PlayerChangedTeam)))
    (:then_3
      (Collision|SetCollisionResponsetoChannel (Variables|Default|GetCollider) "ECC_GameTraceChannel1" "ECR_Ignore"))))

(event Custom|Turn%20Visibility%20Off%20Event
  (params (Instigator :type Controller%20Object%20Reference))
  ;[1 1 1] Device Events
  (Networking|SwitchHasAuthority
    (:Authority
      (Visibility|SetForcedVisibilityDuringGame (Variables|Default|GetVisibleInGameComponent) "NewEnumerator2"))))

(event Custom|Respawn%20Event
  (params (Instigator :type Player%20Controller%20Object%20Reference))
  ;[1 1 1] Device Events
  (Utilities|RemoveCollectorfromArray (CallFunction|GetPlayerStatefromTargetActor Instigator))
  (CallFunction|ServerUpdateforPlayers))

(event Custom|Respawn%20For%20All%20Event
  ;[1 1 1] Device Events
  (Utilities|Set|Clear (Variables|Default|GetCollectedByPlayers))
  (Variables|Default|SetbRequiresServerUpdate true)
  (CallFunction|ServerUpdateforPlayers))

(event OnVisibilityChanged(VisibleInGameComponent)
  (params (Visible :type Boolean))
  ;[1 1 1] On visibility changed
  (CallFunction|UpdateCosmeticsforLocalPlayers Visible))

(event MinigameEndedDelegate(FortMinigameProgress)
  (bind _minigame (Utilities|GetMinigame))
  ;[1 1 1] On Minigame Ended
  (Utilities|FlowControl|Sequence
    (:then_0
      (CallFunction|SetVisibility true))
    (:then_1
      (Collision|SetCollisionEnabled (Variables|Default|GetCollider) "QueryOnly")
      (Collision|SetCollisionResponsetoChannel (Variables|Default|GetCollider) "ECC_GameTraceChannel1" "ECR_Block"))
    (:then_2
      (Utilities|Set|Clear (Variables|Default|GetCollectedByPlayers))
      (Utilities|Map|Clear (Variables|Default|GetClientCoinVisibility))
      (Networking|SwitchHasAuthority
        (:Authority
          (Visibility|SetForcedVisibilityDuringGame (Variables|Default|GetVisibleInGameComponent) "NewEnumerator1")
          (Utilities|Array|Clear (Variables|Default|GetHideFromPlayers)))))
    (:then_3
      (Components|Activation|SetActive (Variables|Default|GetIdleFX) (Variables|UserOptions|GetPlayAmbientVFX) false)
      (CallFunction|UpdateAmbientAudio true)
      (Variables|Default|SetPickupFlash 0))
    (:then_4
      (Minigame|UnbindEventfromPlayerAddedDelegate _minigame (EventDispatchers|CreateEvent :!event Custom|On%20Player%20Added))
      (Minigame|UnbindEventfromPlayerRemovedDelegate _minigame (EventDispatchers|CreateEvent :!event Custom|On%20Player%20Removed))
      (Minigame|UnbindEventfromOnPlayerClassApplied _minigame (EventDispatchers|CreateEvent :!event Custom|On%20Player%20Class%20Changed))
      (Minigame|UnbindEventfromOnPlayerClassRemoved _minigame (EventDispatchers|CreateEvent :!event Custom|On%20Player%20Class%20Changed))
      (Minigame|UnbindEventfromOnPlayerJoininProgress _minigame (EventDispatchers|CreateEvent :!event Custom|On%20Player%20Joined%20In%20Progress))
      (Minigame|UnbindEventfromPlayerChangedTeamDelegate _minigame (EventDispatchers|CreateEvent :!event PlayerChangedTeam)))
    (:then_5
      (Variables|Default|SetbRequiresServerUpdate true))
    (:then_6
      (Variables|Default|SetCachedCollectionPickupVFX 0) ; If we use custom FX, clear it in case\nwe're in creative and change the asset between\nsessions.
    )))

(event EventEndPlay
  (params (EndPlayReason :type EEndPlayReason%20Enum :default "Destroyed"))
  ;[1 1 1] End Play Cleanup - Unbind Team Change
  (for _array_element (Fortnite|GetLocalFortPlayerControllers true true)
    (bind _asfort_player_state_athena (Utilities|Casting|CastToFortPlayerStateAthena (Variables|/Script/Engine.Controller|GetPlayerState _array_element)))
    (Player|UnbindallEventsfromOnTeamIndexChangedDelegate _asfort_player_state_athena)))

(event EventDestroyed
  (bind _audio_looping (Variables|Default|GetAudio%20Looping))
  ;[1 1 1] Event Destroyed Cleanup
  (Networking|CanExecuteCosmeticEvents
    (:True
      (if (Audio|Components|Audio|IsPlaying _audio_looping)
        (Audio|Components|Audio|FadeOut _audio_looping 0.5 0.0 "Linear" :!class /Script/Engine.AudioComponent)))))

(fn CanPlayerPickUp
  (params (FortPlayerState :type Fort%20Player%20State%20Object%20Reference))
  (bind _returnvalue (PlayerState|GetUniqueID FortPlayerState))
  (bind _clientcoinvisibility (Variables|Default|GetClientCoinVisibility))
  (bind (_value _returnvalue_2) (Utilities|Map|Find _clientcoinvisibility _returnvalue))
  (bind (_playerid_4_95a294f943dd14b55003f293fad0d8f7 _coincollectibility_19_33ff1a6246850380d3a27985748043b1 _coinvisibility_2_0aa632be4e5700a1fb276f90c334d421) (Utilities|Struct|BreakPlayerCoinVisibility _value))
  (Networking|SwitchHasAuthority
    (:Authority
      (if (Visibility|ShouldbeVisible (Variables|Default|GetVisibleInGameComponent))
        (Utilities|IsValid FortPlayerState
          (:Is%20Valid
            (if _returnvalue_2
              (return (and _coincollectibility_19_33ff1a6246850380d3a27985748043b1 _coinvisibility_2_0aa632be4e5700a1fb276f90c334d421))
              (else
                (return false))))
          (:Is%20Not%20Valid
            _))
        (else
          _))
      (return false))
    (:Remote
      _)))

(fn CheckClassRules
  (params (Player%20Actor :type Actor%20Object%20Reference))
  (bind _valid_controller (CreativeUserRestrictions|ActorHasValidParticipantController (Variables|Default|GetCreative_UseRestriction_Component) Player%20Actor true true true))
  (return _valid_controller))

(fn CheckInstanceRules
  (params (CurrentPlayerId :type Unique%20Net%20Id%20Structure%20%28by%20ref%29) (CollectedTeamCounts :type Map%20of%20Integers%20to%20Integers) (CurrentPlayerTeam :type Integer%20%28by%20ref%29))
  (bind (_value _returnvalue) (Utilities|Map|Find (Variables|GetCollectedTeamCounts) (Variables|GetCurrentPlayerTeam)))
  (switch Utilities|FlowControl|Switch|SwitchonECollectibleObjectConsumedIf (Variables|UserOptions|Basic|GetConsumeIfCollectedBy)
    (:NewEnumerator0
      (return (not (Utilities|Set|ContainsItem (Variables|Default|GetCollectedByPlayers) (Variables|GetCurrentPlayerId)))))
    (:NewEnumerator1
      (return (<= (Utilities|Set|Length (Variables|Default|GetCollectedByPlayers)) 0)))
    (:NewEnumerator2
      (return (<= (select _returnvalue _value 0) 0)))))

(fn CheckNonCollectingTeamVisibilityRules
  (params (CapturedPerTeam :type Map%20of%20Integers%20to%20Integers) (TotalPerTeam :type Map%20of%20Integers%20to%20Integers) (CurrentPlayerCollectingTeam :type Integer))
  (bind (_value _returnvalue) (Utilities|Map|Find (Variables|GetCapturedPerTeam) (CallFunction|GetCollectingTeam)))
  (bind _collecting_team (CallFunction|GetCollectingTeam))
  (bind _totalperteam (Variables|Default|GetTotalPerTeam))
  (bind (_value_2 _returnvalue_2) (Utilities|Map|Find _totalperteam _collecting_team))
  (bind _capturedperteam (Variables|GetCapturedPerTeam))
  (bind (_value_3 _returnvalue_3) (Utilities|Map|Find _capturedperteam _collecting_team))
  (if (and (Knobs|Team|IsValidTeamIndex (Creative|TeamOption|GetLegacyTeamIndex (Variables|UserOptions|GetCollectingTeam))) (!= (CallFunction|GetCollectingTeam) CurrentPlayerCollectingTeam))
    (switch Utilities|FlowControl|Switch|SwitchonECollectibleVisibleToOpposing (Variables|UserOptions|GetVisibleToOpposingPlayers)
      (:NewEnumerator0
        (return false))
      (:NewEnumerator1
        (return true))
      (:NewEnumerator2
        (switch Utilities|FlowControl|Switch|SwitchonECollectibleObjectConsumedIf (Variables|UserOptions|Basic|GetConsumeIfCollectedBy)
          (:NewEnumerator0
            (if _returnvalue_3
              (if _returnvalue_2
                (return (< _value_3 _value_2))
                (else
                  (return false)))
              (else
                (return true))))
          (:NewEnumerator1
            _)
          (:NewEnumerator2
            _))
        (return (<= (select _returnvalue _value 0) 0))))
    (else
      (return false))))

(fn CheckTeamRules
  (params (CurrentPlayerTeam :type Integer))
  (return (or (Creative|TeamOption|IsTeamType (Variables|UserOptions|GetCollectingTeam) "All") (== (CallFunction|GetCollectingTeam) CurrentPlayerTeam))))

(fn DoCollection
  (params (ForActor :type Actor%20Object%20Reference))
  (bind _playerstateofcollector (LocalVariables|GetPlayerStateOfCollector))
  (bind _asfort_vehicle_interface (Utilities|Casting|CastToFortVehicleInterface ForActor)
    (:then
      (bind _playerstate (Utilities|GetPlayerStatefromActor (Vehicle|GetDriver _asfort_vehicle_interface))
        (:Is%20Valid
          (LocalVariables|SetPlayerStateOfCollector _playerstate)
          (if (not (CallFunction|CanPlayerPickUp (LocalVariables|GetPlayerStateOfCollector)))
            (return false)))
        (:Is%20Not%20Valid
          ;[1 1 1] Only used for debugging coin pick ups
          (if (not (and (Variables|Debug|GetbSelfPickUp) (Utilities|Equal(Object) ForActor self)))
            (return false)))) ; Get driver
      (Utilities|AddCollectortoArray _playerstateofcollector)
      (CallFunction|ServerUpdateforPlayers)
      (CallFunction|SendTrigger _playerstateofcollector)
      (CallFunction|SendStatEvents (LocalVariables|GetPlayerStateOfCollector))
      (return true))
    (:CastFailed
      (bind _playerstate_2 (Utilities|GetPlayerStatefromActor ForActor)
        (:Is%20Valid
          (LocalVariables|SetPlayerStateOfCollector _playerstate_2))
        (:Is%20Not%20Valid
          _)) ; Make sure it was a player
    )))

(fn GetPlayerStateFromTargetActor
  (params (Actor :type Actor%20Object%20Reference))
  (bind _controller (CreativeAI|Controller|GetParticipantControllerfromActor Actor))
  (bind _asfort_player_controller_spectating (Utilities|Casting|CastToFortPlayerControllerSpectating Actor)
    (:then
      (return (Class|FortPlayerControllerSpectating|GetFollowedPlayer _asfort_player_controller_spectating)))
    (:CastFailed
      (Utilities|IsValid _controller
        (:Is%20Valid
          (bind _playerstate (Variables|/Script/Engine.Controller|GetPlayerState _controller)
            (:then
              (bind _asfort_player_state (Utilities|Casting|CastToFortPlayerState _playerstate)
                (:then
                  (return _asfort_player_state))
                (:CastFailed
                  _)))
            (:else
              _))
          (return 0))
        (:Is%20Not%20Valid
          _)))))

(fn GetCollectibleObject
  (return (Variables|UserOptions|Basic|GetCollectibleObject)))

(fn GetCollectingTeam
  (return (Math|Conversions|ToInteger(Byte) (Creative|TeamOption|ConverttoTeam (Variables|UserOptions|GetCollectingTeam)))))

(fn GetOverrideMeshMaterial
  (params (MeshComp :type Static%20Mesh%20Component%20Object%20Reference) (MatIdx :type Integer))
  (if (and (Utilities|Equal(Object) MeshComp (Variables|Default|GetVisibleMesh)) (== MatIdx 0))
    (return (Variables|Default|GetCurrentMeshComponentMaterial))
    (else
      (return (|CallParentFunction MeshComp MatIdx)))))

(fn InitializeVisibilityComponent
  (bind _visibleingamecomponent (Variables|Default|GetVisibleInGameComponent))
  (Visibility|SetVisibilityComponents (Variables|Default|GetVisibleInGameComponent) (Utilities|Array|MakeArray (Variables|Debug|GetStaticMeshComponent) (Variables|Default|GetVisibleMesh) (Variables|Default|GetIdleFX)) true)
  (Networking|SwitchHasAuthority
    (:Authority
      (Visibility|SetVisiblityDuringPhase _visibleingamecomponent (select (Variables|UserOptions|GetVisibleOnGameStart) "NewEnumerator3" "NewEnumerator1"))
      (Visibility|SetForcedVisibilityDuringGame _visibleingamecomponent "NewEnumerator0"))))

(fn IsTargetableByLinearMover
  (return true))

(fn MeshPickerUpgradePath
  (bind (_mesh _scale _stattag) (Utilities|Struct|BreakCreativeDeviceMeshSettings (Utilities|Array|Get(aref) (Variables|Default|GetCoinMeshScaleTagArray) (Variables|Default|GetMeshPicker))))
  (bind _meshpicker (Variables|Default|GetMeshPicker))
  (bind _coinmeshscaletagarray (Variables|Default|GetCoinMeshScaleTagArray))
  (if (Utilities|Array|IsValidIndex _coinmeshscaletagarray _meshpicker)
    (LocalVariables|SetTempMeshPicker _meshpicker)
    (Variables|Default|SetCurrentCoinTag _stattag)
    (Variables|Default|SetMeshPicker -1)
    (OptionData|SetPropertyOverrideValues (Variables|Default|GetToyOptionsComponent) (Utilities|Map|MakeMap "MeshIndex" (Utilities|String|ToString(Integer) _meshpicker)))))

(fn MessageScore
  (return (Variables|UserOptions|Advanced|GetHUD%20Message) (Variables|UserOptions|Advanced|GetHUD%20MessageScoreColor) (Variables|UserOptions|Advanced|GetHUD%20MessageColor) (Variables|UserOptions|Advanced|GetResetHUD%20MessageScore)))

(fn OnlyVisibleToCollectingTeamUpgradePath
  (bind _toyoptionscomponent (Variables|Default|GetToyOptionsComponent))
  (bind (_currentoptions _returnvalue) (OptionData|GetPropertyOverrides _toyoptionscomponent))
  (bind (_value _returnvalue_2) (Utilities|Map|Find _currentoptions "bOnlyVisibleToCollectingTeam"))
  (bind _toyoptionscomponent_2 (Variables|Default|GetToyOptionsComponent))
  (Networking|SwitchHasAuthority
    (:Authority
      (if _returnvalue
        (if _returnvalue_2
          (bind _output_get (LocalVariables|SetCurrentVisibleTeamPropertyOverride _value))
          (bind _output_get_2 (LocalVariables|SetCurrentVisibleTeamActualValue (Utilities|String|Equal,CaseInsensitive(String) _output_get "TRUE")))
          (bind _output_get_3 (Variables|UserOptions|SetVisibleToOpposingPlayers (select _output_get_2 "NewEnumerator0" "NewEnumerator1")))
          (OptionData|SetPropertyOverrides _toyoptionscomponent_2 (Utilities|Map|MakeMap (Class|PlaylistUserOptionBase|GetOptionKey (Variables|Default|GetVisibleToNonCollectionTeamOption) false 0) (Class|PlaylistUserOptionBase|GetOptionValueNamefromValue (Variables|Default|GetVisibleToNonCollectionTeamOption) (Utilities|String|ToString(Integer) (Math|Conversions|ToInteger(Byte) _output_get_3)) true 0)))
          (OptionData|RemovePropertyOverrides _toyoptionscomponent_2 (Utilities|Array|MakeArray "bOnlyVisibleToCollectingTeam")))))))

(fn OnRep_bAmbientAudio
  (Networking|CanExecuteCosmeticEvents
    (:True
      (for _array_element (Fortnite|GetLocalFortPlayerControllers true true)
        (bind _out_unique_id (Utilities|GetUniqueIdfromActor _array_element)
          (:Is%20Valid
            (if (Utilities|Array|ContainsItem (Variables|Default|GetHideFromPlayers) _out_unique_id)
              (CallFunction|UpdateAmbientAudio false)
              (return)))))
      (CallFunction|UpdateAmbientAudio true))))

(fn OnRep_HideFromPlayers
  (CallFunction|UpdateCosmeticsforLocalPlayers (Visibility|ShouldbeVisible (Variables|Default|GetVisibleInGameComponent))))

(fn OnRep_MeshPicker
  (CallFunction|PickMesh))

(fn PickupColor_ApplyColor
  (params (ColorToApply :type Linear%20Color%20Structure))
  (Utilities|FlowControl|Sequence
    (:then_0
      (CallFunction|PickupColorUpdateIdleFXColor false (Variables|GetColorToApply)))
    (:then_1
      (if (Variables|UserOptions|GetCustomMesh_Override)
        (Rendering|Material|SetVectorParameterValue (Variables|CollectibleColoration|GetVisibleMeshMID) "UniqueColor" (Variables|GetColorToApply) :!class /Script/Engine.MaterialInstanceDynamic)
        (else
          (for _array_element (Utilities|Select :Index (CallFunction|GetCollectibleObject) :NewEnumerator0 (Utilities|Array|MakeArray "ColorOuter" "ColorInner") :NewEnumerator1 (Utilities|Array|MakeArray "EmissiveColor") :NewEnumerator2 (Utilities|Array|MakeArray "BaseColor_RedVert" "EdgeColor" "Rarity_Color") :NewEnumerator3 (Utilities|Array|MakeArray "BaseColor_RedVert" "EdgeColor" "Rarity_Color") :NewEnumerator6 (Utilities|Array|MakeArray "BaseColor_RedVert" "EdgeColor" "Rarity_Color") :NewEnumerator8 (Utilities|Array|MakeArray "Color") :NewEnumerator9 (Utilities|Array|MakeArray "DiffuseColorScale") :NewEnumerator10 (Utilities|Array|MakeArray "EmissiveColor") :NewEnumerator11 (Utilities|Array|MakeArray "BaseDiffuse" "EdgeDiffuse") :NewEnumerator17 (Utilities|Array|MakeArray "BaseColor" "GlowColour" "SweepColour") :NewEnumerator18 (Utilities|Array|MakeArray "ColorOuter" "ColorInner") :NewEnumerator19 (Utilities|Array|MakeArray "ColorOuter" "ColorInner") :NewEnumerator20 (Utilities|Array|MakeArray "ColorOuter" "ColorInner"))
            (Rendering|Material|SetVectorParameterValue (Variables|CollectibleColoration|GetVisibleMeshMID) _array_element (Variables|GetColorToApply) :!class /Script/Engine.MaterialInstanceDynamic)))))) ; 0 - Apply Color to VFX\n1 - Apply Color to Mesh
)

(fn PickupColor_GetColor
  (bind (_base_color _light_color) (Teams|TeamOption|GetTeamColorforTeamOption self (Utilities|Select :Index (Variables|UserOptions|GetCollectibleColor) :NewEnumerator0 (Variables|UserOptions|GetCollectingTeam) :NewEnumerator1 (Variables|UserOptions|GetCollectingTeam) :NewEnumerator2 (Variables|UserOptions|GetCollectingTeam) :NewEnumerator3 (Variables|UserOptions|GetCollectingTeam) :NewEnumerator4 (Variables|UserOptions|GetSpecificTeamIndex) :NewEnumerator5 (Variables|UserOptions|GetSpecificTeamIndex))))
  (switch Utilities|FlowControl|Switch|SwitchonECollectibleObjectColorOption (Variables|UserOptions|GetCollectibleColor)
    (:NewEnumerator0
      (return "(R=1.000000,G=1.000000,B=1.000000,A=1.000000)"))
    (:NewEnumerator1
      (return (Variables|UserOptions|GetCustomColor)))
    (:NewEnumerator2
      _)
    (:NewEnumerator3
      (bind _valid_controller (CreativeUserRestrictions|ActorHasValidParticipantController (Variables|Default|GetCreative_UseRestriction_Component) (Fortnite|GetFirstLocalFortPlayerController true) true true true))
      (return (Math|Color|SelectColor (Variables|CollectibleColoration|GetFriendlyColor) (Variables|CollectibleColoration|GetEnemyColor) (or (and _valid_controller (Utilities|Enum|Equal(Enum) (Team|GetTeamAffiliationBetweenTeams (Team|GetActorTeam (Fortnite|GetFirstLocalFortPlayerController true)) (Creative|TeamOption|ConverttoTeam (Utilities|Select :Index (Variables|UserOptions|GetCollectibleColor) :NewEnumerator0 (Variables|UserOptions|GetCollectingTeam) :NewEnumerator1 (Variables|UserOptions|GetCollectingTeam) :NewEnumerator2 (Variables|UserOptions|GetCollectingTeam) :NewEnumerator3 (Variables|UserOptions|GetCollectingTeam) :NewEnumerator4 (Variables|UserOptions|GetSpecificTeamIndex) :NewEnumerator5 (Variables|UserOptions|GetSpecificTeamIndex)))) "Friendly")) (Creative|TeamOption|IsTeamType (Utilities|Select :Index (Variables|UserOptions|GetCollectibleColor) :NewEnumerator0 (Variables|UserOptions|GetCollectingTeam) :NewEnumerator1 (Variables|UserOptions|GetCollectingTeam) :NewEnumerator2 (Variables|UserOptions|GetCollectingTeam) :NewEnumerator3 (Variables|UserOptions|GetCollectingTeam) :NewEnumerator4 (Variables|UserOptions|GetSpecificTeamIndex) :NewEnumerator5 (Variables|UserOptions|GetSpecificTeamIndex)) "All")))))
    (:NewEnumerator4
      _)
    (:NewEnumerator5
      _))
  (return (Math|Color|SelectColor "(R=1.000000,G=1.000000,B=1.000000,A=1.000000)" _light_color (Creative|TeamOption|IsTeamType (Utilities|Select :Index (Variables|UserOptions|GetCollectibleColor) :NewEnumerator0 (Variables|UserOptions|GetCollectingTeam) :NewEnumerator1 (Variables|UserOptions|GetCollectingTeam) :NewEnumerator2 (Variables|UserOptions|GetCollectingTeam) :NewEnumerator3 (Variables|UserOptions|GetCollectingTeam) :NewEnumerator4 (Variables|UserOptions|GetSpecificTeamIndex) :NewEnumerator5 (Variables|UserOptions|GetSpecificTeamIndex)) "All"))))

(fn PickupColor_SetupMID
  (Variables|CollectibleColoration|SetVisibleMeshMID 0)
  (Material|ReattachOrCreateMID (Variables|Default|GetVisibleMesh) (Variables|CollectibleColoration|GetVisibleMeshMID)))

(fn PickupColor_UpdateIdleFXColor
  (params (ResetToDefault :type Boolean) (ColorToApply :type Linear%20Color%20Structure))
  (bind _resettodefault (Variables|GetResetToDefault))
  (bind _defaultidlevfxcolor (Variables|CollectibleColoration|GetDefaultIdleVFXColor))
  (bind _colortoapply (Variables|GetColorToApply))
  (bind _returnvalue (select _resettodefault _defaultidlevfxcolor _colortoapply))
  (Effects|Components|ParticleSystem|SetColorParameter (Variables|Default|GetIdleFX) "GlowColor" _returnvalue)
  (Effects|Components|ParticleSystem|SetVectorParameter (Variables|Default|GetIdleFX) "LightColor" (* (Math|Conversions|ToVector(LinearColor) _returnvalue) 1e+01)))

(fn PlayerChangedTeam
  (params (PlayerState :type Fort%20Player%20State%20Object%20Reference) (NewTeam :type Byte) (OldTeam :type Byte) (ChangedWithoutRespawn :type Boolean))
  (if (!= NewTeam OldTeam)
    (Variables|Default|SetbRequiresServerUpdate true)
    (CallFunction|ServerUpdateforPlayers)))

(fn SendStatEvents
  (params (InFortPlayerState :type Fort%20Player%20State%20Object%20Reference))
  (bind (_outsourcetags _outcontexttags) (Class|FortMinigameStatsComponent|GetSourceandContextTags (LocalVariables|Getlocal%20MinigameStatsComp) (LocalVariables|GetForPlayerController)))
  (bind _asfort_player_controller_zone (Utilities|Casting|CastToFortPlayerControllerZone (Actor|GetOwner InFortPlayerState)))
  (LocalVariables|SetForPlayerController _asfort_player_controller_zone)
  (bind _output_get (LocalVariables|SetTargetTags (GameplayTags|GetOwnedGameplayTags self :!class /Script/GameplayTags.GameplayTagAssetInterface)))
  (GameplayTags|AddGameplayTag _output_get (Variables|Default|GetCurrentCoinTag) :!class /Script/GameplayTags.BlueprintGameplayTagLibrary)
  (Utilities|FlowControl|Sequence
    (:then_0
      ;[0 0 0 0.5] Broadcast custom verb event for Quest System 2 compatibility
      (Networking|SwitchHasAuthority
        (:Authority
          (CustomVerb|BroadcastCustomVerbMessage (CustomVerb|MakeCustomVerbMessage (LocalVariables|GetForPlayerController) "(TagName=\"CustomVerb.Collect.Collectible\")")))))
    (:then_1
      (bind _output_get_2 (LocalVariables|SetQuestManager (Quests|GetQuestManager (LocalVariables|GetForPlayerController) "Athena")))
      (Utilities|IsValid _output_get_2
        (:Is%20Valid
          (Quest|SendComplexCustomStatEvent (LocalVariables|GetQuestManager) self "(GameplayTags=)" (LocalVariables|GetTargetTags) (Variables|Default|GetAmountToCount)))))
    (:then_2
      (bind _returnvalue (CreativeDevices|GetMinigameStatsComp (LocalVariables|GetForPlayerController) "Athena"))
      (LocalVariables|Setlocal%20MinigameStatsComp _returnvalue)
      (bind _local_minigamestatscomp (LocalVariables|Getlocal%20MinigameStatsComp)
        (:then
          (Class|FortMinigameStatsComponent|SendStatEventwithTags _local_minigamestatscomp (LocalVariables|GetForPlayerController) "Collect" self (LocalVariables|GetTargetTags) (Class|FortMinigameStatsComponent|GetSourceandContextTags _local_minigamestatscomp (LocalVariables|GetForPlayerController)) (.OutContextTags (Class|FortMinigameStatsComponent|GetSourceandContextTags _local_minigamestatscomp (LocalVariables|GetForPlayerController))) (Variables|Default|GetAmountToCount))
          (GameplayTags|AddGameplayTag (LocalVariables|GetTargetTags) "(TagName=\"Minigame.Stat.Score.Coin\")" :!class /Script/GameplayTags.BlueprintGameplayTagLibrary)
          (Class|FortMinigameStatsComponent|SendStatEventwithTags (LocalVariables|Getlocal%20MinigameStatsComp) (LocalVariables|GetForPlayerController) "MinigameDynamicEvent" self (LocalVariables|GetTargetTags) _outsourcetags _outcontexttags (Variables|UserOptions|Basic|GetScore))
          (Class|CreativeHUDBroadcastComponent|DoScoreEvent (Variables|Default|GetCreative_HUDBroadcast_Component) (LocalVariables|GetForPlayerController) (Variables|UserOptions|Basic|GetScore) (Variables|UserOptions|Advanced|GetDisplayScoreUpdateOnHUD)))
        (:else)))))

(fn SendTrigger
  (params (PlayerState :type Fort%20Player%20State%20Object%20Reference))
  (GameplayEvent|BroadcastEvent (Variables|UserOptions|Event|GetOn%20Collected) 0 (Utilities|Struct|MakeGameplayEventPickedUp (Utilities|Casting|CastToController (Actor|GetOwner PlayerState)))))

(fn ServerUpdateForPlayers
  (Networking|SwitchHasAuthority
    (:Authority
      (if (Variables|Default|GetbRequiresServerUpdate)
        (Utilities|Map|Clear (Variables|Default|GetCapturePerTeam))
        (Utilities|Map|Clear (Variables|Default|GetTotalPerTeam))
        (bind _currentminigame (Variables|/Script/FortniteGame.FortMinigameLogicComponent|GetCurrentMinigame (Variables|Default|GetFortMinigameProgress))
          (:then
            (bind _output_get (LocalVariables|SetParticipatingPlayers (Minigame|GetParticipatingPlayers _currentminigame)))
            (for _array_element _output_get
              (bind _output_get_2 (LocalVariables|SetCurrentPlayer _array_element))
              (Utilities|IsValid _output_get_2
                (:Is%20Valid
                  (LocalVariables|SetCurrentPlayerTeam (Math|Conversions|ToInteger(Byte) (Team|GetActorTeam _output_get_2)))
                  (bind _output_get_3 (LocalVariables|SetCurrentPlayerId (PlayerState|GetUniqueID (LocalVariables|GetCurrentPlayer))))
                  (LocalVariables|SetCurrentPlayerCaptured (Utilities|Set|ContainsItem (Variables|Default|GetCollectedByPlayers) _output_get_3))
                  (Utilities|IncrementTeamIntMap (Variables|Default|GetTotalPerTeam) (LocalVariables|GetCurrentPlayerTeam))
                  (if (LocalVariables|GetCurrentPlayerCaptured)
                    (Utilities|IncrementTeamIntMap (Variables|Default|GetCapturePerTeam) (LocalVariables|GetCurrentPlayerTeam))))))
            (Utilities|Map|Clear (Variables|Default|GetClientCoinVisibility))
            (LocalVariables|SetHideFromPlayersChanged false)
            (for _array_element_2 (LocalVariables|GetParticipatingPlayers)
              (bind _output_get_4 (LocalVariables|SetCurrentPlayer _array_element_2))
              (Utilities|IsValid _output_get_4
                (:Is%20Valid
                  (bind _output_get_5 (LocalVariables|SetCurrentPlayerTeam (Math|Conversions|ToInteger(Byte) (Team|GetActorTeam _output_get_4))))
                  (bind _output_get_6 (LocalVariables|SetCurrentPlayerId (PlayerState|GetUniqueID _output_get_4)))
                  (bind _hidefromplayerschanged (CallFunction|UpdateIndividualPlayer _output_get_6 _output_get_5 (LocalVariables|GetCurrentPlayer) (Variables|Default|GetHideFromPlayers)))
                  (if _hidefromplayerschanged
                    (LocalVariables|SetHideFromPlayersChanged true)))))
            (Variables|Default|SetbRequiresServerUpdate false)
            (if (LocalVariables|GetHideFromPlayersChanged)
              (Variables|Default|SetHideFromPlayers (Variables|Default|GetHideFromPlayers))))
          (:else))))))

(fn SetVisibility
  (params (bVisible :type Boolean))
  (Rendering|SetVisibility (Variables|Debug|GetStaticMeshComponent) (Variables|GetbVisible) false)
  (Rendering|SetVisibility (Variables|Default|GetIdleFX) (and bVisible (Variables|UserOptions|GetPlayAmbientVFX)) false))

(fn ShouldInvisiblePropsBeAllowed
  (return false))

(fn UpdateAmbientAudio
  (params (bCanLocalPlayerPickup :type Boolean))
  (bind _audio_looping (Variables|Default|GetAudio%20Looping))
  (if (and bCanLocalPlayerPickup (Variables|UserOptions|Basic|GetbAmbientAudio))
    (Audio|Components|Audio|Play _audio_looping 0.0)
    (else
      (Audio|Components|Audio|Stop _audio_looping))))

(fn UpdateCosmetics
  (params (IsVisible :type Boolean))
  (Networking|CanExecuteCosmeticEvents
    (:True
      (Utilities|FlowControl|Sequence
        (:then_0
          (CallFunction|SetVisibility IsVisible)
          (Components|Activation|SetActive (Variables|Default|GetIdleFX) (and IsVisible (Variables|UserOptions|GetPlayAmbientVFX)) false))
        (:then_1
          (CallFunction|UpdateAmbientAudio IsVisible))))))

(fn UpdateCosmeticsForLocalPlayers
  (params (Visibility :type Boolean))
  (for _array_element (Fortnite|GetLocalFortPlayerControllers true true)
    (bind _out_unique_id (Utilities|GetUniqueIdfromActor _array_element)
      (:Is%20Valid
        (if (Utilities|Array|ContainsItem (Variables|Default|GetHideFromPlayers) _out_unique_id)
          (CallFunction|UpdateCosmetics false)
          (return)))))
  (CallFunction|UpdateCosmetics (Variables|GetVisibility)))

(fn UpdateIndividualPlayer
  (params (Player%20Id :type Unique%20Net%20Id%20Structure) (Player%20Team :type Integer) (Player%20State :type Fort%20Player%20State%20Object%20Reference) (Hide%20From%20Players :type Array%20of%20Unique%20Net%20Id%20Structures))
  (bind _new_collectibility (LocalVariables|GetNew%20Collectibility))
  (bind _player_id (Variables|GetPlayer%20Id))
  (bind _hide_from_players (Variables|GetHide%20From%20Players))
  (LocalVariables|SetNew%20Collectibility (Math|Boolean|ANDBoolean (CallFunction|CheckInstanceRules Player%20Id (Variables|Default|GetCapturePerTeam) (Variables|GetPlayer%20Team)) (Default|CheckTeamRules (Variables|GetPlayer%20Team)) (Default|CheckClassRules Player%20State)))
  (if _new_collectibility
    (LocalVariables|SetNew%20Visibility _new_collectibility)
    (else
      (LocalVariables|SetNew%20Visibility (CallFunction|CheckNonCollectingTeamVisibilityRules (Variables|Default|GetCapturePerTeam) (Variables|Default|GetTotalPerTeam) (Variables|GetPlayer%20Team)))))
  (Utilities|Map|Add (Variables|Default|GetClientCoinVisibility) (Variables|GetPlayer%20Id) (Utilities|Struct|MakePlayerCoinVisibility (Variables|GetPlayer%20Id) (LocalVariables|GetNew%20Collectibility) (LocalVariables|GetNew%20Visibility)))
  (if (LocalVariables|GetNew%20Visibility)
    (bind _returnvalue (Utilities|Array|RemoveItem _hide_from_players _player_id))
    (return _returnvalue)
    (else
      (bind _returnvalue_2 (Utilities|Array|AddUnique _hide_from_players _player_id))
      (return (!= _returnvalue_2 -1)))))

(fn UpdateIndividualPlayerWithNotifyOnChange
  (params (Player%20State :type Fort%20Player%20State%20Object%20Reference))
  (Networking|SwitchHasAuthority
    (:Authority
      (Utilities|IsValid Player%20State
        (:Is%20Valid
          (bind _hidefromplayerschanged (CallFunction|UpdateIndividualPlayer (PlayerState|GetUniqueID Player%20State) (Math|Conversions|ToInteger(Byte) (Team|GetActorTeam Player%20State)) Player%20State (Variables|Default|GetHideFromPlayers)))
          ;[1 1 1] Only Update Hide from Players if we've actually added to the array.
          (if _hidefromplayerschanged
            (Variables|Default|SetHideFromPlayers (Variables|Default|GetHideFromPlayers))))))))

(fn UpdateVisibilityOnGameStart
  (Networking|SwitchHasAuthority
    (:Authority
      (Visibility|SetForcedVisibilityDuringGame (Variables|Default|GetVisibleInGameComponent) (select (Variables|UserOptions|GetVisibleOnGameStart) "NewEnumerator1" "NewEnumerator2")))))

(fn ConstructionScript
  (Variables|/Script/Engine.SceneComponent|SetbAbsoluteScale true (Variables|Default|GetIdleFX))
  (CallFunction|PickMesh))
