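; Creative device for a team color-changing tile that players claim by contact during a minigame, awarding score and optionally stealing it from the previous owner.
; Box overlaps or alternate-shape hits reach StartChangingColor, which sets team state on authority, sends stat events and can start a revert timer.
; Position and team replicate packed in one LinearColor, whose OnRep plays cosmetic FX.
; Options asynchronously load meshes and materials, with replicated glow and icon toggles.
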(event Custom|DebugColorChangeCube
  (params (TeamNumber :type Integer))
  (bind (_returnvalue_x _returnvalue_y _returnvalue_z) (Math|Vector|BreakVector (Transformation|GetActorLocation)))
  ;[1 0.7517 0.03507] Debug
  (Variables|Default|SetCubeCurrentTeamColorVisuals TeamNumber)
  (Utilities|FlowControl|Delay (Math|Random|RandomFloatinRange 0.0 2.0))
  (Variables|Default|SetOverlappedPawnPositionAndTeamNumber (Utilities|Struct|MakeLinearColor _returnvalue_x _returnvalue_y _returnvalue_z (Math|Conversions|ToFloat(Integer) TeamNumber))))

(event EventPostFinishSpawning
  (|CallParentFunction))

(event Custom|Refresh%20Team%20Colors
  (bind _startingteam (Variables|UserOptions|GetStartingTeam))
  (bind _returnvalue (Creative|TeamOption|GetLegacyTeamIndex _startingteam))
  (if (Knobs|Team|IsValidTeamIndex _returnvalue)
    (CallFunction|UpdateTeamColors _returnvalue)))

(event Custom|Update%20Appearance
  (bind (_returnvalue_x _returnvalue_y _returnvalue_z) (Math|Vector|BreakVector (CallFunction|GetDefaultTriggerLocation)))
  (CallFunction|UpdateMesh)
  (if (not (Variables|UserOptions|Advanced|GetAlternateTileShape_Override))
    (Networking|CanExecuteCosmeticEvents
      (:True
        (if (and (Utilities|Array|IsValidIndex (Variables|Materials|GetMaterials) (Math|Conversions|ToInteger(Byte) (Variables|UserOptions|GetAppearance))) (Utilities|Enum|NotEqual(Enum) (Variables|UserOptions|GetAppearance) (Variables|Default|GetLastAppearance)))
          (Variables|Default|SetLastAppearance (Variables|UserOptions|GetAppearance))
          (bind _loaded_object (Creative|Loading|AsyncLoadEditorSafe (Utilities|Array|Get(acopy) (Variables|Materials|GetMaterials) (Math|Conversions|ToInteger(Byte) (Variables|UserOptions|GetAppearance))))
            (:Loaded
              (bind _asmaterial (Utilities|Casting|CastToMaterial _loaded_object))
              (bind _output_get (Variables|Default|SetCubeMID (Material|GetMIDfromMaterial (Variables|Default|GetCubeMesh) 0 _asmaterial)))
              (Rendering|Material|SetMaterial (Variables|Default|GetCubeMesh) 0 _output_get)
              ;[1 1 1] If we are in the editor then we will want to set the team color again after setting the visuals
              (if (not (Editor|InEditor))
                (CallFunction|PlayVisualFX)
                (else
                  (Variables|Default|SetOverlappedPawnPositionAndTeamNumber (Utilities|Struct|MakeLinearColor _returnvalue_x _returnvalue_y _returnvalue_z (Math|Conversions|ToFloat(Integer) (Creative|TeamOption|GetLegacyTeamIndex (Variables|UserOptions|GetStartingTeam))))))))))))))

(event Custom|RevertToDefaultColor
  (bind (_returnvalue_x _returnvalue_y _returnvalue_z) (Math|Vector|BreakVector (CallFunction|GetDefaultTriggerLocation)))
  ;[1 1 1] Update Visuals On Options Changed
  (Networking|SwitchHasAuthority
    (:Authority
      (if (not (== (Variables|Default|GetCubeCurrentTeamColorVisuals) (Creative|TeamOption|GetLegacyTeamIndex (Variables|UserOptions|GetStartingTeam))))
        (Variables|Default|SetCubeCurrentTeamColorVisuals (Creative|TeamOption|GetLegacyTeamIndex (Variables|UserOptions|GetStartingTeam)))
        (Variables|Default|SetOverlappedPawnPositionAndTeamNumber (Utilities|Struct|MakeLinearColor _returnvalue_x _returnvalue_y _returnvalue_z (Math|Conversions|ToFloat(Integer) (Creative|TeamOption|GetLegacyTeamIndex (Variables|UserOptions|GetStartingTeam)))))))
    (:Remote
      (CallFunction|RefreshTeamColors)
      (CallFunction|PlayVisualFX))))

(event OnAnyPropertyChanged(ToyOptionsComponent)
  (bind (_returnvalue_x _returnvalue_y _returnvalue_z) (Math|Vector|BreakVector (CallFunction|GetDefaultTriggerLocation)))
  ;[1 1 1] Update Visuals On Options Changed
  (Class|CreativeUIBroadcastComponent|Initialize (Variables|Default|GetCreative_HUDBroadcast_Component) (Variables|Default|GetFortMinigameProgress) (Variables|UserOptions|Advanced|GetDisplayScoreUpdateOnHUD))
  (Variables|Default|SetColorChangeTransitionTime (Variables|Default|GetDefaultColorChangeTransitionTime))
  (Visibility|SetVisiblityDuringPhase (Variables|Default|GetVisibleInGameComponent) (select (Variables|UserOptions|GetVisibleDuringGame) "NewEnumerator3" "NewEnumerator1"))
  (CallFunction|UpdateCollision)
  (CallFunction|UpdateAppearance)
  (Networking|SwitchHasAuthority
    (:Authority
      (if (not (== (Variables|Default|GetCubeCurrentTeamColorVisuals) (Creative|TeamOption|GetLegacyTeamIndex (Variables|UserOptions|GetStartingTeam))))
        (Variables|Default|SetCubeCurrentTeamColorVisuals (Creative|TeamOption|GetLegacyTeamIndex (Variables|UserOptions|GetStartingTeam)))
        (Variables|Default|SetOverlappedPawnPositionAndTeamNumber (Utilities|Struct|MakeLinearColor _returnvalue_x _returnvalue_y _returnvalue_z (Math|Conversions|ToFloat(Integer) (Creative|TeamOption|GetLegacyTeamIndex (Variables|UserOptions|GetStartingTeam)))))))
    (:Remote
      (CallFunction|RefreshTeamColors)
      (CallFunction|PlayVisualFX))))

(event OnPropertiesLoaded(ToyOptionsComponent)
  (bind (_returnvalue_x _returnvalue_y _returnvalue_z) (Math|Vector|BreakVector (CallFunction|GetDefaultTriggerLocation)))
  ;[1 1 1] Update Visuals On Options Changed
  (Class|CreativeUIBroadcastComponent|Initialize (Variables|Default|GetCreative_HUDBroadcast_Component) (Variables|Default|GetFortMinigameProgress) (Variables|UserOptions|Advanced|GetDisplayScoreUpdateOnHUD))
  (Variables|Default|SetColorChangeTransitionTime (Variables|Default|GetDefaultColorChangeTransitionTime))
  (Visibility|SetVisiblityDuringPhase (Variables|Default|GetVisibleInGameComponent) (select (Variables|UserOptions|GetVisibleDuringGame) "NewEnumerator3" "NewEnumerator1"))
  (CallFunction|UpdateCollision)
  (CallFunction|UpdateAppearance)
  (Networking|SwitchHasAuthority
    (:Authority
      (if (not (== (Variables|Default|GetCubeCurrentTeamColorVisuals) (Creative|TeamOption|GetLegacyTeamIndex (Variables|UserOptions|GetStartingTeam))))
        (Variables|Default|SetCubeCurrentTeamColorVisuals (Creative|TeamOption|GetLegacyTeamIndex (Variables|UserOptions|GetStartingTeam)))
        (Variables|Default|SetOverlappedPawnPositionAndTeamNumber (Utilities|Struct|MakeLinearColor _returnvalue_x _returnvalue_y _returnvalue_z (Math|Conversions|ToFloat(Integer) (Creative|TeamOption|GetLegacyTeamIndex (Variables|UserOptions|GetStartingTeam)))))))
    (:Remote
      (CallFunction|RefreshTeamColors)
      (CallFunction|PlayVisualFX))))

(event OnComponentBeginOverlap(OverlapBox)
  (params (OverlappedComponent :type Primitive%20Component%20Object%20Reference) (OtherActor :type Actor%20Object%20Reference) (OtherComp :type Primitive%20Component%20Object%20Reference) (OtherBodyIndex :type Integer) (bFromSweep :type Boolean) (SweepResult :type Hit%20Result%20Structure%20%28by%20ref%29))
  (CallFunction|StartChangingColor OtherActor))

(event Custom|Start%20Reverting
  ;[1 1 1] After time spent waiting before we begin to revert color (if option enabled)
  (Variables|Default|SetColorChangeTransitionTime (Math|Conversions|ToFloat(Integer) (Variables|Default|GetRevertColorTransition)))
  (CallFunction|DeductScorefromOlderPawn (Variables|Default|GetOverlappedPawn))
  (Variables|Default|SetOverlappedPawn 0)
  (CallFunction|ReverttoDefaultColor))

(event Custom|HideCubeRays
  ;[1 1 1] HideCubeRays
  (Rendering|SetVisibility (Variables|Default|GetSm_CubeColorChangeRays) false false))

(event MinigameStateChangedDelegate(FortMinigameLogic)
  (params (Minigame :type Fort%20Minigame%20Object%20Reference) (NewMinigameState :type EFortMinigameState%20Enum :default "PreGame"))
  ;[1 1 1] Reset visuals to neutral state between games
  (if (Utilities|Select :Index NewMinigameState :PreGame false :Setup false :Transitioning true :WaitingForCameras false :Warmup true :InProgress false :PostGameTimeDilation false :PostRoundEnd false :PostGameEnd false :PostGameAbandon false :PostGameReset true :Teardown false)
    (CallFunction|UpdateTeamColors (Creative|TeamOption|GetLegacyTeamIndex (Variables|UserOptions|GetStartingTeam)))
    (Networking|SwitchHasAuthority
      (:Authority
        (CallFunction|ClearAllTimerHandles)
        (Variables|Default|SetOverlappedPawn 0)
        (Variables|Default|SetEnabled (Variables|UserOptions|GetEnabledAtGameStart))
        (Variables|Default|SetEmissiveEnabled false)
        (Variables|Default|SetIconVisible (Variables|UserOptions|Advanced|GetIconVisibleAtGameStart)))
      (:Remote
        (CallFunction|HideCubeRays)))
    (elif (Utilities|Enum|Equal(Enum) NewMinigameState "PreGame")
      (Variables|Default|SetIconVisible true))))

(event OnComponentHit(CubeMesh)
  (params (HitComponent :type Primitive%20Component%20Object%20Reference) (OtherActor :type Actor%20Object%20Reference) (OtherComp :type Primitive%20Component%20Object%20Reference) (NormalImpulse :type Vector :default "0, 0, 0") (Hit :type Hit%20Result%20Structure%20%28by%20ref%29))
  (if (Variables|UserOptions|Advanced|GetAlternateTileShape_Override)
    (CallFunction|StartChangingColor OtherActor)))

(event Custom|Enable%20Event
  (params (Instigator :type Controller%20Object%20Reference))
  ;[0.6731 0 1] Gameplay Functions
  (Variables|Default|SetEnabled true))

(event Custom|Disable%20Event
  (params (Instigator :type Controller%20Object%20Reference))
  ;[0.6731 0 1] Gameplay Functions
  (Variables|Default|SetEnabled false))

(event Custom|Reset%20Event
  (params (Instigator :type Controller%20Object%20Reference))
  ;[0.6731 0 1] Gameplay Functions
  (CallFunction|ClearAllTimerHandles)
  (CallFunction|StartReverting))

(event Custom|Set%20Event
  (params (Instigator :type Controller%20Object%20Reference))
  (bind _returnvalue (Pawn|GetControlledPawn Instigator))
  ;[0.6731 0 1] Gameplay Functions
  (if (Variables|Default|GetEnabled)
    (Utilities|IsValid _returnvalue
      (:Is%20Valid
        (CallFunction|StartChangingColor _returnvalue false)))))

(event Custom|Show%20Event
  (params (Instigator :type Controller%20Object%20Reference))
  ;[0.6731 0 1] Gameplay Functions
  (Visibility|SetForcedVisibilityDuringGame (Variables|Default|GetVisibleInGameComponent) "NewEnumerator1"))

(event Custom|Hide%20Event
  (params (Instigator :type Controller%20Object%20Reference))
  ;[0.6731 0 1] Gameplay Functions
  (Visibility|SetForcedVisibilityDuringGame (Variables|Default|GetVisibleInGameComponent) "NewEnumerator2"))

(event Custom|ActivateGlowEvent
  (params (Instigator :type Controller%20Object%20Reference))
  ;[0.6731 0 1] Gameplay Functions
  (if (Variables|Default|GetEnabled)
    (Variables|Default|SetEmissiveEnabled true)))

(event Custom|DeactivateGlowEvent
  (params (Instigator :type Controller%20Object%20Reference))
  ;[0.6731 0 1] Gameplay Functions
  (if (Variables|Default|GetEnabled)
    (Variables|Default|SetEmissiveEnabled false)))

(event Custom|ShowIconEvent
  (params (Instigator :type Controller%20Object%20Reference))
  ;[0.6731 0 1] Gameplay Functions
  (Variables|Default|SetIconVisible true))

(event Custom|HideIconEvent
  (params (Instigator :type Controller%20Object%20Reference))
  ;[0.6731 0 1] Gameplay Functions
  (Variables|Default|SetIconVisible false))

(event OnVisibilityChanged(VisibleInGameComponent)
  (params (Visible :type Boolean))
  (CallFunction|UpdateCollision))

(event Custom|AdjustEmissiveGlow
  (params (Play%20Forward :type Boolean))
  (bind (_emissivecolor_r _emissivecolor_g _emissivecolor_b) (Math|Color|BreakColor (Variables|GetEmissiveColor_R)))
  ;[1 1 1] Adjust Emissive
  (Material|ReattachOrCreateMID (Variables|Default|GetCubeMesh) (Variables|Default|GetSecondaryColorMID) 1)
  (if Play%20Forward
    (bind (_direction _emissive__multiplier) (Timeline|Timeline Play)
      (:Update
        (bind _output_get (Variables|SetEmissiveColor_R (Math|Color|MakeColor _emissivecolor_r _emissivecolor_g _emissivecolor_b _emissive__multiplier)))
        (Rendering|Material|SetVectorParameterValue (Variables|Default|GetSecondaryColorMID) "Emissive" _output_get :!class /Script/Engine.MaterialInstanceDynamic)))
    (else
      (Timeline|Timeline Reverse))))

(event Custom|ShowOrHideIcon
  (params (ShowIcon :type Boolean))
  (bind _iconmid (Variables|Default|GetIconMID))
  (bind (_alternate_tile_shape_icon_smallicon _alternate_tile_shape_icon_largeicon) (Utilities|Struct|BreakCreativeIconOption (Variables|GetAlternate%20Tile%20Shape%20Icon_SmallIcon)))
  ;[1 1 1] Show Or Hide Icon
  (bind _loaded_object (Creative|Loading|AsyncLoadEditorSafe (select ShowIcon _alternate_tile_shape_icon_largeicon (Variables|Default|GetEmptyIcon)))
    (:Loaded
      (bind _astexture_2d (Utilities|Casting|CastToTexture2D _loaded_object))
      (Material|ReattachOrCreateMID (Variables|Default|GetCubeMesh) _iconmid 2)
      (Rendering|Material|SetTextureParameterValue _iconmid "Texture Sample" _astexture_2d)
      (Rendering|Material|SetVectorParameterValue (Variables|Default|GetIconMID) "Color_Tint" (Variables|UserOptions|Advanced|GetAlternate%20Tile%20Shape%20Starting%20Primary%20Color) :!class /Script/Engine.MaterialInstanceDynamic))))

(event Custom|UpdateMesh
  (bind (_overlappedpawnpositionandteamnumber_r _overlappedpawnpositionandteamnumber_g _overlappedpawnpositionandteamnumber_b _overlappedpawnpositionandteamnumber_a) (Math|Color|BreakColor (Variables|GetOverlappedPawnPositionAndTeamNumber_R)))
  (bind _cubemesh (Variables|Default|GetCubeMesh))
  (bind (_alternate_tile_shape_starting_top_color_r _alternate_tile_shape_starting_top_color_g _alternate_tile_shape_starting_top_color_b) (Math|Color|BreakColor (Variables|GetAlternate%20Tile%20Shape%20Starting%20Top%20Color_R)))
  ;[1 1 1] Update to use Custom Mesh
  (bind _loaded_object (Creative|Loading|AsyncLoadEditorSafe (select (Variables|UserOptions|Advanced|GetAlternateTileShape_Override) (Variables|UserOptions|Advanced|GetAlternateTileShape) "/Game/Playgrounds/Items/Cubes/Cube_Variants/S_Cube_4x4x1.S_Cube_4x4x1"))
    (:Loaded
      (bind _asstatic_mesh (Utilities|Casting|CastToStaticMesh _loaded_object))
      (Components|StaticMesh|SetStaticMesh _cubemesh _asstatic_mesh)
      (Transformation|SetRelativeLocation _cubemesh (select (Variables|UserOptions|Advanced|GetAlternateTileShape_Override) "-256.000000,-256.000000,0.000000" "0, 0, 0") false false)
      (if (Variables|UserOptions|Advanced|GetAlternateTileShape_Override)
        (bind _loaded_object_2 (Creative|Loading|AsyncLoadEditorSafe (Variables|Default|GetHexMaterialReference))
          (:Loaded
            (bind _asmaterial_instance (Utilities|Casting|CastToMaterialInstance _loaded_object_2))
            (Rendering|Material|SetMaterial _cubemesh 0 _asmaterial_instance)
            (Material|ReattachOrCreateMID _cubemesh (Variables|Default|GetPrimaryColorMID))
            (Material|ReattachOrCreateMID _cubemesh (Variables|Default|GetSecondaryColorMID) 1)
            (Rendering|Material|SetVectorParameterValue (Variables|Default|GetPrimaryColorMID) "TintColor" (Variables|UserOptions|Advanced|GetAlternate%20Tile%20Shape%20Starting%20Primary%20Color) :!class /Script/Engine.MaterialInstanceDynamic)
            (Rendering|Material|SetVectorParameterValue (Variables|Default|GetSecondaryColorMID) "TintColor" (Math|Color|SelectColor (CallFunction|GetColorforTeam (Math|Float|Truncate _overlappedpawnpositionandteamnumber_a)) (Variables|UserOptions|Advanced|GetAlternate%20Tile%20Shape%20Starting%20Top%20Color) (CallFunction|ShouldUseTeamColor)) :!class /Script/Engine.MaterialInstanceDynamic)
            (bind _output_get (Variables|SetStartEmissiveColor_R (Math|Color|MakeColor _alternate_tile_shape_starting_top_color_r _alternate_tile_shape_starting_top_color_g _alternate_tile_shape_starting_top_color_b (Variables|Default|GetStartingEmissiveMultiplier))))
            (bind _output_get_2 (Variables|Default|SetEmissiveColor _output_get))
            ;[1 1 1] Definitively disabling collision on the base mesh as it can sometimes be enabled after scaling
            (Rendering|Material|SetVectorParameterValue (Variables|Default|GetSecondaryColorMID) "Emissive" _output_get_2 :!class /Script/Engine.MaterialInstanceDynamic)
            (CallFunction|ShoworHideIcon (Variables|Default|GetIconVisible))))))))

(fn ClearAllTimerHandles
  (Utilities|Time|ClearandInvalidateTimerbyHandle (Variables|Default|GetTimeBeforeRevertingTimerHandle))
  (Utilities|Time|ClearandInvalidateTimerbyHandle (Variables|Default|GetHideCubeRaysTimer)))

(fn DeductScoreFromOlderPawn
  (params (Older%20Pawn :type Fort%20Player%20Pawn%20Object%20Reference))
  (if (Variables|UserOptions|Basic|GetStealScore)
    (Utilities|IsValid Older%20Pawn
      (:Is%20Valid
        (bind _output_get (LocalVariables|SetTargetTags (GameplayTags|GetOwnedGameplayTags self :!class /Script/GameplayTags.GameplayTagAssetInterface)))
        (GameplayTags|AddGameplayTag _output_get (Variables|Default|GetTagToSend) :!class /Script/GameplayTags.BlueprintGameplayTagLibrary)
        (GameplayTags|AddGameplayTag (LocalVariables|GetTargetTags) "(TagName=\"Minigame.Stat.Score.Coin\")" :!class /Script/GameplayTags.BlueprintGameplayTagLibrary)
        (CallFunction|SendScoreEvent (LocalVariables|GetTargetTags) Older%20Pawn (* (Variables|UserOptions|Basic|GetScore) -1))))))

(fn FindNormalizedOverlapPointLocalSpace
  (bind (_origin _boxextent) (Collision|GetComponentBounds (Variables|Default|GetCubeMesh)))
  (bind (_returnvalue_location _returnvalue_rotation _returnvalue_scale) (Math|Transform|BreakTransform (Transformation|GetWorldTransform (Variables|Default|GetCubeMesh))))
  (bind (_overlappedpawnpositionandteamnumber_r _overlappedpawnpositionandteamnumber_g _overlappedpawnpositionandteamnumber_b) (Math|Color|BreakColor (Variables|GetOverlappedPawnPositionAndTeamNumber_R)))
  (return (/ (Math|Vector|VectorGetAbs (Math|Transform|InverseTransformLocation (Transformation|GetWorldTransform (Variables|Default|GetCubeMesh)) (Math|Vector|MakeVector _overlappedpawnpositionandteamnumber_r _overlappedpawnpositionandteamnumber_g _overlappedpawnpositionandteamnumber_b))) (* (Math|Vector|VectorGetAbs (Math|Transform|InverseTransformLocation (Math|Transform|MakeTransform "0, 0, 0" _returnvalue_rotation _returnvalue_scale) _boxextent)) 2.0))))

(fn GetDefaultTriggerLocation
  (bind (_origin _boxextent) (Collision|GetComponentBounds (Variables|Default|GetCubeMesh)))
  (return (+ (Transformation|GetWorldLocation (Variables|Default|GetCubeMesh)) (Math|Vector|VectorGetAbs (Math|Transform|TransformDirection (Transformation|GetWorldTransform (Variables|Default|GetCubeMesh)) _boxextent)))))

(fn GetColorForTeam
  (params (Team :type Integer))
  (bind _self self)
  (bind _returnvalue (Minigame|GetMinigame _self :!class /Script/FortniteGame.MinigameBlueprintLibrary))
  (bind _team (Variables|Team|GetTeam))
  (bind _returnvalue_2 (Knobs|Team|GetTeamValueasIndex _team))
  (bind _returnvalue_3 (Minigame|GetMinigameTeamComponent _returnvalue))
  (bind (_teamlightcolors _teamselectorcolors) (Class|GetClassDefaults "/Game/Creative/Sets/MilitaryBase/Props/CreativeTeamColors.CreativeTeamColors_C"))
  (Utilities|IsValid _returnvalue
    (:Is%20Valid
      (Utilities|IsValid _returnvalue_3
        (:Is%20Valid
          (LocalVariables|SetColorIndex (Minigame|GameRules|GetTeamColorIndex _returnvalue_3 (Teams|IndextoTeam _returnvalue_2))))
        (:Is%20Not%20Valid
          (LocalVariables|SetColorIndex (+ _returnvalue_2 1) :!label _label)))
      (bind _output_get (LocalVariables|SetCurrentTeamSelectorColors _teamselectorcolors))
      (if (Utilities|Enum|Equal(Enum) (Variables|UserOptions|GetAppearance) "NewEnumerator1")
        (Utilities|Array|SetArrayElem (LocalVariables|GetCurrentTeamSelectorColors) 0 (Variables|Default|GetOff%20Color)))
      (return (Utilities|Array|Get(aref) (LocalVariables|GetCurrentTeamSelectorColors) (select (and (Utilities|Array|IsValidIndex _output_get (Math|Integer|Max(Integer) (Math|Integer|%25%28Integer%29 (LocalVariables|GetColorIndex) (Utilities|Array|Length _teamselectorcolors)))) (Knobs|Team|IsValidTeamIndex (Variables|Team|GetTeam))) (Math|Integer|Max(Integer) (Math|Integer|%25%28Integer%29 (LocalVariables|GetColorIndex) (Utilities|Array|Length _teamselectorcolors))) 0))))
    (:Is%20Not%20Valid
      _))
  (goto _label))

(fn GetOverrideMeshMaterial
  (params (MeshComp :type Static%20Mesh%20Component%20Object%20Reference) (MatIdx :type Integer))
  (Utilities|FlowControl|Sequence
    (:then_0
      (if (Utilities|Equal(Object) (Variables|GetMeshComp) (Variables|Default|GetCubeMesh))
        (return (Variables|Default|GetCubeMID))))
    (:then_1
      (if (Utilities|Equal(Object) (Variables|Default|GetSm_CubeColorChangeRays) (Variables|Default|GetCubeMesh))
        (return (Variables|Default|GetRaysMID))))))

(fn GetPawnTeamNumber
  (params (Actor :type Actor%20Object%20Reference))
  (return (- (Math|Conversions|ToInteger(Byte) (Team|GetActorTeam Actor)) 2)))

(fn GetRaysMID
  (Material|ReattachOrCreateMID (Variables|Default|GetSm_CubeColorChangeRays) (Variables|Default|GetRaysMID))
  (return (Variables|Default|GetRaysMID)))

(fn IsTargetableByLinearMover
  (return true))

(fn MessageScore
  (return (Variables|UserOptions|Advanced|GetHUD%20Message) (Variables|UserOptions|Advanced|GetHUD%20MessageScoreColor) (Variables|UserOptions|Advanced|GetHUD%20MessageColor) (Variables|UserOptions|Advanced|GetResetHUD%20MessageScore)))

(fn MigrateLegacyUserOptions
  (bind _timebeforereverting (Class|GetClassDefaults "/CreativeCoreDevices/Device_ColorChangingTiles_V2.Device_ColorChangingTiles_V2_C"))
  (|CallParentFunction)
  (Utilities|FlowControl|Sequence
    (:then_0
      (Utilities|FlowControl|Sequence
        (:then_0
          ;[0.3313 0.9502 1] Migrating TimeBeforeReverting -> Splitting into IsTileReverting(Never,Time), TileRevertingTimeUnits(Seconds,Minutes), TimeUntilReverting(int)
          (if (!= _timebeforereverting (Variables|Legacy|GetTimeBeforeReverting))
            (Variables|UserOptions|Basic|SetRevertTile true)
            (Variables|UserOptions|Basic|SetTimeUntilReverting (Variables|Legacy|GetTimeBeforeReverting))
            (Variables|Legacy|SetTimeBeforeReverting _timebeforereverting)
            (LocalVariables|SetMigratedProperties true)))))
    (:then_1
      (if (LocalVariables|GetMigratedProperties)
        (OptionData|RefreshPropertyOverrides (Variables|Default|GetToyOptionsComponent))))))

(fn OnRep_EmissiveEnabled
  (Networking|CanExecuteCosmeticEvents
    (:True
      (CallFunction|AdjustEmissiveGlow (Variables|Default|GetEmissiveEnabled)))))

(fn OnRep_IconVisible
  (Networking|CanExecuteCosmeticEvents
    (:True
      (CallFunction|ShoworHideIcon (Variables|Default|GetIconVisible)))))

(fn OnRep_OverlappedPawnPositionAndTeamNumber
  (Networking|CanExecuteCosmeticEvents
    (:True
      (CallFunction|PlayVisualFX))))

(fn SendCollectionEvent
  (LocalVariables|SetFortPlayerController (Fortnite|GetFortPlayerControllerfromActor (Variables|Default|GetOverlappedPawn)))
  (bind _output_get (LocalVariables|SetTargetTags (GameplayTags|GetOwnedGameplayTags self :!class /Script/GameplayTags.GameplayTagAssetInterface)))
  (GameplayTags|AddGameplayTag _output_get (Variables|Default|GetTagToSend) :!class /Script/GameplayTags.BlueprintGameplayTagLibrary)
  (Utilities|FlowControl|Sequence
    (:then_0
      (bind _output_get_2 (LocalVariables|SetQuestManager (Quests|GetQuestManager (LocalVariables|GetFortPlayerController) "Athena")))
      (Utilities|IsValid _output_get_2
        (:Is%20Valid
          (Quest|SendComplexCustomStatEvent (LocalVariables|GetQuestManager) self "(GameplayTags=)" (LocalVariables|GetTargetTags) (Variables|Default|GetAmountToCount)))))
    (:then_1
      (bind _returnvalue (CreativeDevices|GetMinigameStatsComp (LocalVariables|GetFortPlayerController) "Athena"))
      (Utilities|IsValid _returnvalue
        (:Is%20Valid
          (Class|FortMinigameStatsComponent|SendStatEventwithTags _returnvalue (LocalVariables|GetFortPlayerController) "Collect" self (LocalVariables|GetTargetTags) (Class|FortMinigameStatsComponent|GetSourceandContextTags _returnvalue (LocalVariables|GetFortPlayerController)) (.OutContextTags (Class|FortMinigameStatsComponent|GetSourceandContextTags _returnvalue (LocalVariables|GetFortPlayerController))) (Variables|Default|GetAmountToCount))
          (GameplayTags|AddGameplayTag (LocalVariables|GetTargetTags) "(TagName=\"Minigame.Stat.Score.Coin\")" :!class /Script/GameplayTags.BlueprintGameplayTagLibrary)
          (CallFunction|SendScoreEvent (LocalVariables|GetTargetTags) (Variables|Default|GetOverlappedPawn) (Variables|UserOptions|Basic|GetScore))
          (Class|CreativeHUDBroadcastComponent|DoScoreEventforParticipant (Variables|Default|GetCreative_HUDBroadcast_Component) (Fortnite|GetFortPlayerControllerfromActor (Variables|Default|GetOverlappedPawn)) (Variables|UserOptions|Basic|GetScore) (Variables|UserOptions|Advanced|GetDisplayScoreUpdateOnHUD)))))
    (:then_2
      (GameplayEvent|BroadcastEvent (Variables|UserOptions|Events|GetOn%20Activated) 0 (Utilities|Struct|MakeGameplayEventActivated (LocalVariables|GetFortPlayerController))))))

(fn SendScoreEvent
  (params (Tags :type Gameplay%20Tag%20Container%20Structure) (Pawn%20To%20Give%20Score :type Fort%20Player%20Pawn%20Object%20Reference) (Score :type Integer))
  (bind _returnvalue (Fortnite|GetFortPlayerControllerfromActor Pawn%20To%20Give%20Score))
  (Utilities|IsValid Pawn%20To%20Give%20Score
    (:Is%20Valid
      (Utilities|IsValid _returnvalue
        (:Is%20Valid
          (bind _returnvalue_2 (CreativeDevices|GetMinigameStatsComp _returnvalue "Athena"))
          (Utilities|IsValid _returnvalue_2
            (:Is%20Valid
              (Class|FortMinigameStatsComponent|SendStatEventwithTags _returnvalue_2 _returnvalue "MinigameDynamicEvent" self Tags (Class|FortMinigameStatsComponent|GetSourceandContextTags _returnvalue_2 _returnvalue) (.OutContextTags (Class|FortMinigameStatsComponent|GetSourceandContextTags _returnvalue_2 _returnvalue)) Score))))))))

(fn SetTransitionTimestampParameters
  (params (TimestampA :type Float%20%28double-precision%29) (TransitionLengthIn :type Float%20%28double-precision%29) (TimestampB :type Float%20%28double-precision%29) (TransitionLengthOut :type Float%20%28double-precision%29))
  (if (not (Variables|UserOptions|Advanced|GetAlternateTileShape_Override))
    (Material|ReattachOrCreateMID (Variables|Default|GetCubeMesh) (Variables|Default|GetCubeMID))
    (Rendering|Material|SetVectorParameterValue (Variables|Default|GetCubeMID) "CubeTimestampPayload" (Math|Color|MakeColor (Variables|GetTimestampA) (Variables|GetTransitionLengthIn) (Variables|GetTimestampB) (Variables|GetTransitionLengthOut)) :!class /Script/Engine.MaterialInstanceDynamic)
    (Rendering|Material|SetVectorParameterValue (CallFunction|GetRaysMID) "CubeTimestampPayload" (Math|Color|MakeColor (Variables|GetTimestampA) (Variables|GetTransitionLengthIn) (Variables|GetTimestampB) (Variables|GetTransitionLengthOut)) :!class /Script/Engine.MaterialInstanceDynamic)))

(fn ShouldInvisiblePropsBeAllowed
  (return false))

(fn ShouldUseTeamColor
  (bind _startingteam_teamtype (Utilities|Struct|BreakCreativeTeamOption (Variables|GetStartingTeam_TeamType)))
  (return (or (Utilities|Enum|Equal(Enum) _startingteam_teamtype "TeamIndex") (and (not (Utilities|Enum|Equal(Enum) _startingteam_teamtype "TeamIndex")) (!= (Variables|Default|GetCubeCurrentTeamColorVisuals) 990)))))

(fn StartChangingColor
  (params (Overlapping%20Pawn :type Actor%20Object%20Reference) (Use%20Pawn%20Location%20for%20FX :type Boolean :default true))
  (bind (_returnvalue_x _returnvalue_y _returnvalue_z) (Math|Vector|BreakVector (Math|Vector|SelectVector (Transformation|GetActorLocation (Variables|Default|GetOverlappedPawn)) (CallFunction|GetDefaultTriggerLocation) (Variables|GetUse%20Pawn%20Location%20for%20FX))))
  (bind _local_pawn (LocalVariables|Getlocal%20Pawn))
  (Networking|SwitchHasAuthority
    (:Authority
      (if (Variables|Default|GetEnabled)
        (bind _asfort_player_pawn (Utilities|Casting|CastToFortPlayerPawn Overlapping%20Pawn))
        (LocalVariables|Setlocal%20Pawn _asfort_player_pawn)
        (Utilities|IsValid (Variables|/Script/FortniteGame.FortMinigameLogicComponent|GetCurrentMinigame (Variables|Default|GetFortMinigameLogic))
          (:Is%20Valid
            (if (not (or (== (Knobs|Team|GetTeamValue (Variables|Default|GetCubeCurrentTeamColorVisuals)) (CallFunction|GetPawnTeamNumber (LocalVariables|Getlocal%20Pawn))) (Utilities|Select :Index (Variables|/Script/FortniteGame.FortMinigame|GetCurrentState (Variables|/Script/FortniteGame.FortMinigameLogicComponent|GetCurrentMinigame (Variables|Default|GetFortMinigameLogic))) :PreGame true :Setup true :Transitioning true :WaitingForCameras true :Warmup true :InProgress false :PostGameTimeDilation false :PostRoundEnd true :PostGameEnd true :PostGameAbandon true :PostGameReset true :Teardown false)))
              (CallFunction|DeductScorefromOlderPawn (Variables|Default|GetOverlappedPawn))
              (Variables|Default|SetOverlappedPawn _local_pawn)
              (bind _output_get (Variables|Default|SetCubeCurrentTeamColorVisuals (Knobs|Team|ReshiftTeamValue (CallFunction|GetPawnTeamNumber _local_pawn))))
              (if (Variables|UserOptions|Advanced|GetChangeColorOnPlayerContact)
                (Variables|Default|SetOverlappedPawnPositionAndTeamNumber (Utilities|Struct|MakeLinearColor _returnvalue_x _returnvalue_y _returnvalue_z (Math|Conversions|ToFloat(Integer) _output_get))))
              (Variables|Default|SetColorChangeTransitionTime (Variables|Default|GetDefaultColorChangeTransitionTime))
              (if (Variables|UserOptions|Basic|GetRevertTile)
                (bind _returnvalue (Utilities|Time|SetTimerbyEvent (EventDispatchers|CreateEvent :!event Start%20Reverting) (Math|Conversions|ToFloat(Integer) (Variables|UserOptions|Basic|GetTimeUntilReverting))))
                (Variables|Default|SetTimeBeforeRevertingTimerHandle _returnvalue))
              (CallFunction|SendCollectionEvent))))))))

(fn UpdateCollision
  (bind (_value _returnvalue) (Utilities|Map|Find (Variables|Default|GetMeshToOverlapLocationMap) (Variables|UserOptions|Advanced|GetAlternateTileShape)))
  (bind (_value_2 _returnvalue_2) (Utilities|Map|Find (Variables|Default|GetMeshToOverlapExtentMap) (Variables|UserOptions|Advanced|GetAlternateTileShape)))
  (bind _alternatetileshape_override (Variables|UserOptions|Advanced|GetAlternateTileShape_Override))
  (bind _overlapbox (Variables|Default|GetOverlapBox))
  (Utilities|FlowControl|Sequence
    (:then_0
      (if (Class|CreativeCommonDeviceFunctionLibrary|IsinPlayMode self)
        (switch Utilities|FlowControl|Switch|SwitchonEnum_CollisionDuringGamesSetting (Variables|UserOptions|GetCollisionDuringGames)
          (:NewEnumerator0
            (bind _output_get (LocalVariables|SetCollisionProfileName (CreativeDevices|GetNamefromCollisionProfile "(Name=\"FortCreativeOnlyVolumeManager\")")))
            (LocalVariables|SetOverlapCollisionProfileName _output_get))
          (:NewEnumerator1
            (LocalVariables|SetCollisionProfileName (CreativeDevices|GetNamefromCollisionProfile "(Name=\"FortBuildingMeshPhysics\")"))
            (LocalVariables|SetOverlapCollisionProfileName (CreativeDevices|GetNamefromCollisionProfile "(Name=\"OverlapOnlyPawn\")")))
          (:NewEnumerator3
            (if (Visibility|ShouldbeVisible (Variables|Default|GetVisibleInGameComponent))
              (LocalVariables|SetCollisionProfileName (CreativeDevices|GetNamefromCollisionProfile "(Name=\"FortBuildingMeshPhysics\")"))
              (LocalVariables|SetOverlapCollisionProfileName (CreativeDevices|GetNamefromCollisionProfile "(Name=\"OverlapOnlyPawn\")"))
              (else
                (bind _output_get_2 (LocalVariables|SetCollisionProfileName (CreativeDevices|GetNamefromCollisionProfile "(Name=\"FortCreativeOnlyVolumeManager\")")))
                (LocalVariables|SetOverlapCollisionProfileName _output_get_2)))))
        (else
          (LocalVariables|SetCollisionProfileName (CreativeDevices|GetNamefromCollisionProfile "(Name=\"FortBuildingMeshPhysics\")"))
          (LocalVariables|SetOverlapCollisionProfileName (CreativeDevices|GetNamefromCollisionProfile "(Name=\"OverlapOnlyPawn\")")))))
    (:then_1
      (Collision|SetCollisionProfileName (Variables|Default|GetCubeMesh) (LocalVariables|GetCollisionProfileName) true :!class /Script/Engine.PrimitiveComponent)
      (Collision|SetCollisionProfileName (Variables|Default|GetOverlapBox) (LocalVariables|GetOverlapCollisionProfileName) true :!class /Script/Engine.PrimitiveComponent))
    (:then_2
      ;[1 1 1] Adjust size of Overlap Volume if we are using hex meshes
      (Components|Box|SetBoxExtent _overlapbox (select (and _returnvalue_2 _alternatetileshape_override) _value_2 "260.000000,260.000000,64.000000") true)
      (Transformation|SetRelativeLocation _overlapbox (select (and _returnvalue _alternatetileshape_override) _value "-256.000000,-256.000000,60.000000") false false))))

(fn UpdateTeamColors
  (params (Team%20Color :type Integer))
  (Variables|Default|SetCubeCurrentTeamColorVisuals Team%20Color)
  (Networking|CanExecuteCosmeticEvents
    (:True
      (CallFunction|SetTeamColorDestination Team%20Color)
      (CallFunction|SetTransitionTimestampParameters 0.0 0.1 1e+08 0.1))))

(fn ConstructionScript
  (|CallParentFunction)
  (Variables|Default|SetRaysMID (Material|GetMID (Variables|Default|GetSm_CubeColorChangeRays)))
  (CallFunction|UpdateAppearance))
