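; Creative Conditional Button device that activates when a player with the required items, team and class interacts, or via receiver or music sequencer.
; The server checks and unlocks up to three key-lock components, broadcasts success or missing-item events, and handles cooldown, enable state and reset.
; Outside play, dropped pickups set each slot's required item and damage cycles the slot.
; Clients drive proximity-triggered hologram previews, colors, sounds and capture effects, separated by authority/cosmetic branches.
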(event OnAddedtoMinigame(FortMinigameProgress)
  (params (Minigame :type Fort%20Minigame%20Object%20Reference))
  ;[1 0.1082 0 0.25] Disable Collision on drop area during Edit mode, as overlaps are expensive.
  (Minigame|BindEventtoOnPlayModeChangedDelegate Minigame (delegate Custom|CustomEvent))
  (CallFunction|CustomEvent Minigame (Minigame|IsinPlayMode Minigame :!class /Script/FortniteGame.FortMinigame)))

(event Custom|CustomEvent
  (params (Minigame :type Fort%20Minigame%20Object%20Reference) (bIsInPlayMode :type Boolean))
  ;[1 0.1082 0 0.25] Disable Collision on drop area during Edit mode, as overlaps are expensive.
  (Collision|SetCollisionEnabled (Variables|Default|GetDropArea) (select bIsInPlayMode "NoCollision" "QueryOnly")))

(event Custom|Initialize%20Static%20Device%20Settings
  ;[0 1 0.3626 0.25] [CLIENT/SERVER] Hide preview,  bind cosmetic events , and disable non required components
  (Utilities|NotforPreviewing
    (:True
      (Creative|Networking|Authority/Cosmetic
        (:Authority
          (Components|Activation|SetActive (Variables|Default|GetShowPreviewArea) false false))
        (:Cosmetic
          (Collision|SetCollisionEnabled (Variables|Default|GetShowPreviewArea) "QueryOnly")
          (CallFunction|InitializeVisibility)
          ;[0 1 0.3626 0.25] [CLIENT] Player changing teams
          (bind _asfort_player_controller_athena (Utilities|Casting|CastToFortPlayerControllerAthena (Fortnite|GetFirstLocalFortPlayerController true)))
          (FortPlayerControllerAthena|BindEventtoOnLocalPlayerChangedTeams _asfort_player_controller_athena (EventDispatchers|CreateEvent :!event UpdateButtonCollision))
          (Visuals|TickPreviewAnimation 0.0) ; We want set the animation preview to the hide state in one frame\n(Deliberately placed after the Player Controller cast so this doesn't start in editor)
        )))
    (:False
      (Components|DestroyComponent (Variables|Default|GetHologram_Grid)))))

(event OnAnyPropertyChanged(ToyOptionsComponent)
  ;[0 1 0.3626 0.25] [SERVER/CLIENT] Initialization - Update User Options
  (CallFunction|InitializeStaticDeviceSettings)
  (Utilities|NotforPreviewing
    (:True
      (CallFunction|UpdateButtonCollision)
      (CallFunction|InitializeVisibility)
      (CallFunction|AssignToRegistry)
      (Creative|Networking|Authority/Cosmetic
        (:Authority
          (Initialization|SetKeysRequired)
          (Initialization|SetupTeam)
          (Initialization|DeviceReset)
          (CallFunction|RefreshNextItemToAdd)
          (Utilities|Array|Clear (Variables|Default|GetCompletedKeyIndexes))
          (Variables|Gameplay|SetScoreToAdd 0)
          (Variables|Gameplay|SetScoreToAdd1 0)
          (Variables|Gameplay|SetScoreToAdd2 0))
        (:Cosmetic
          (Variables|Gameplay|SetbIsPreviewOn false) ; This makes sure that if you're next to\nit when making changes in Creative,\nthe holograms update immediately
          (Visuals|UpdateDeviceDisplayIcon false)
          (CallFunction|UpdateHolograms)
          (CallFunction|UpdateNextItemToAdd)
          (Class|FortCreativeInteractVisualizationComponent|SetRadius (Variables|Default|GetInteractVisualizationComponent) (Variables|UserOptions|GetInteractionRadius))
          (Initialization|BindEvents true)
          (Visuals|UpdateColors)
          (CallFunction|UpdatePreviewVisibility))))))

(event OnPropertiesLoaded(ToyOptionsComponent)
  ;[0 1 0.3626 0.25] [SERVER/CLIENT] Initialization - Update User Options
  (CallFunction|InitializeStaticDeviceSettings)
  (Utilities|NotforPreviewing
    (:True
      (CallFunction|UpdateButtonCollision)
      (CallFunction|InitializeVisibility)
      (CallFunction|AssignToRegistry)
      (Creative|Networking|Authority/Cosmetic
        (:Authority
          (Initialization|SetKeysRequired)
          (Initialization|SetupTeam)
          (Initialization|DeviceReset)
          (CallFunction|RefreshNextItemToAdd)
          (Utilities|Array|Clear (Variables|Default|GetCompletedKeyIndexes))
          (Variables|Gameplay|SetScoreToAdd 0)
          (Variables|Gameplay|SetScoreToAdd1 0)
          (Variables|Gameplay|SetScoreToAdd2 0))
        (:Cosmetic
          (Variables|Gameplay|SetbIsPreviewOn false) ; This makes sure that if you're next to\nit when making changes in Creative,\nthe holograms update immediately
          (Visuals|UpdateDeviceDisplayIcon false)
          (CallFunction|UpdateHolograms)
          (CallFunction|UpdateNextItemToAdd)
          (Class|FortCreativeInteractVisualizationComponent|SetRadius (Variables|Default|GetInteractVisualizationComponent) (Variables|UserOptions|GetInteractionRadius))
          (Initialization|BindEvents true)
          (Visuals|UpdateColors)
          (CallFunction|UpdatePreviewVisibility))))))

(event EventOnWorldReady
  ;[0 1 0.3626 0.25] [SERVER/CLIENT] Initialization - Update User Options
  (|CallParentFunction))

(event Custom|On%20Retry%20Bind%20Delegates
  (params (Minigame :type Fort%20Minigame%20Object%20Reference))
  ;[0 1 0.3626 0.25] [SERVER/CLIENT] Initialization - Update User Options
  (Initialization|BindEvents true)
  (Visuals|UpdateColors)
  (CallFunction|UpdatePreviewVisibility))

(event EditorPreview|EventSetupForEditorPreview
  ;[0 1 0.3626 0.25] [SERVER/CLIENT] Initialization - Update User Options
  (|CallParentFunction)
  (Initialization|SetKeysRequired))

(event Custom|Switch%20Item
  ;[0 0.3595 0.5 0.25] [SERVER ONLY] Capture the item definition from any pickup dropped close enough
  (Networking|SwitchHasAuthority
    (:Authority
      (Default|IncrementNextItemToAdd))))

(event OnComponentHit(DropArea)
  (params (HitComponent :type Primitive%20Component%20Object%20Reference) (OtherActor :type Actor%20Object%20Reference) (OtherComp :type Primitive%20Component%20Object%20Reference) (NormalImpulse :type Vector :default "0, 0, 0") (Hit :type Hit%20Result%20Structure%20%28by%20ref%29))
  ;[0 0.3595 0.5 0.25] [SERVER ONLY] Capture the item definition from any pickup dropped close enough
  (Networking|SwitchHasAuthority
    (:Authority
      (CallFunction|CaptureItem OtherActor (Variables|Default|GetNextItemToAdd)))))

(event Interaction|EventBlueprintOnInteract
  (params (InteractingPawn :type Fort%20Pawn%20Object%20Reference) (InteractionBeingAttempted :type EInteractionBeingAttempted%20Enum :default "FirstInteraction"))
  ;[0 0.3631 0.8698 0.25] [SERVER ONLY] Attempting to Interact with the Lock Device
  (CallFunction|InteractIfHasKeys InteractingPawn))

(event Custom|Attempt%20Interact
  (params (Pawn :type Pawn%20Object%20Reference))
  ;[0.04933 1 0 0.25] [SERVER ONLY] Send Unlock message
  (bind _asfort_player_pawn (Utilities|Casting|CastToFortPlayerPawn Pawn))
  (if (Math|Boolean|ANDBoolean (or (TestingRequirements|AnyLocked) (Variables|UserOptions|GetRemainUnlockedAfterActivation)) (Variables|Gameplay|GetbIsEnabled) (not (Variables|Gameplay|GetCoolingDown)))
    (bind _returnvalue (ComponentOrders|Lock|UnlockAll _asfort_player_pawn))
    (if _returnvalue
      (Utilities|Array|Clear (Variables|Default|GetCompletedKeyIndexes))
      (GameplayEvent|BroadcastEvent (Variables|UserOptions|Events|GetOn%20Activated) 0 (Utilities|Struct|MakeGameplayEventActivated (Pawn|GetController _asfort_player_pawn)))
      (if (not (Variables|UserOptions|GetRemainUnlockedAfterActivation))
        (Class|FortCreativeKeyLockComponent|Reset (Variables|Default|GetLock%20Array) false))
      (if (Variables|UserOptions|GetDisableAfterUse)
        (CallFunction|SetEnableState false)
        (else
          (Audio|PlaySoundatLocation (Variables|Audio|GetUnlock%20Sound) (Transformation|GetActorLocation))
          ;;[1 1 1] Bypass delay if there is no cooldown
          (if (> (Variables|UserOptions|GetResetDelay) 0.0)
            (Variables|Gameplay|SetCoolingDown true)
            (Utilities|FlowControl|Delay (Variables|UserOptions|GetResetDelay))
            (else
              _))
          (Variables|Gameplay|SetCoolingDown false)
          (if (CallFunction|ToggleonUseEnabled)
            (Variables|Gameplay|SetIconsWereToggled (not (Variables|Gameplay|GetIconsWereToggled)))))))))

(event Custom|Interact%20if%20Has%20Keys
  (params (Pawn :type Pawn%20Object%20Reference))
  ;[0 0.3631 0.8698 0.25] [SERVER ONLY] Attempting to Interact with the Lock Device
  (if (and (TestingRequirements|PawnHasKeys Pawn) (TestingRequirements|PawnIsHoldingItem Pawn))
    (CallFunction|AttemptInteract Pawn)
    (else
      (CallFunction|FailedInteract Pawn))))

(event Custom|Failed%20Interact
  (params (Pawn :type Pawn%20Object%20Reference))
  ;[1 0.1779 0 0.25] Interacting has failed. Send a message on the failed channel.
  (Utilities|IsValid Pawn
    (:Is%20Valid
      (if (Math|Boolean|ANDBoolean (TestingRequirements|AnyLocked) (not (Variables|Gameplay|GetCoolingDown)) (Variables|Gameplay|GetbIsEnabled))
        (GameplayEvent|BroadcastEvent (Variables|UserOptions|Events|GetOn%20Not%20Enough%20Items) 0 (Utilities|Struct|MakeGameplayEventInteractionFailed (Pawn|GetController Pawn)))))))

(event Custom|HandleEnable
  (params (Instigator :type Controller%20Object%20Reference))
  ;[0.323 0 1 0.25] [SERVER ONLY] Transmitters  / Receivers
  (CallFunction|SetEnableState true))

(event EventPostFinishSpawning
  ;[0.323 0 1 0.25] [SERVER ONLY] Transmitters  / Receivers
  (|CallParentFunction)
  (UEFNConversionHelper|MigratetoReceiverwhenSameChannel (Variables|Default|GetEnableReceiver) (Variables|Default|GetDisableReceiver) (Variables|Default|GetToggle%20Receiver)))

(event Custom|HandleDisable
  (params (Instigator :type Controller%20Object%20Reference))
  ;[0.323 0 1 0.25] [SERVER ONLY] Transmitters  / Receivers
  (CallFunction|SetEnableState false))

(event Custom|HandleToggle
  (params (Instigator :type Controller%20Object%20Reference))
  ;[0.323 0 1 0.25] [SERVER ONLY] Transmitters  / Receivers
  (CallFunction|SetEnableState (not (Variables|Gameplay|GetbIsEnabled))))

(event MinigameStartedDelegate(FortMinigameProgress)
  ;[0 0.8698 0.7627 0.25] [SERVER/CLIENT] Handle minigame events
  (Variables|Gameplay|SetbIsMinigameInProgress true)
  (CallFunction|UpdateButtonCollision)
  (Networking|CanExecuteCosmeticEvents
    (:True
      (CallFunction|UpdateNextItemToAdd)
      (Visuals|UpdateColors)
      (CallFunction|UpdatePreviewVisibility)
      (Default|CallOnGameStateChanged (Variables|Gameplay|GetbIsMinigameInProgress))
      (Class|BPcLockHologramV2|UpdateScreenColorfromLocalPawn (Variables|Default|GetHologram%20Array)) ; Need to update screen colors for holograms\nat beginning of minigame as screen colors use\nin progress state as part of their logic.
    )
    (:False
      (Utilities|HandleDeviceReset))))

(event Custom|HandleReset
  (params (Instigator :type Controller%20Object%20Reference))
  ;[0.323 0 1 0.25] [SERVER ONLY] Transmitters  / Receivers
  (Utilities|HandleDeviceReset))

(event MinigameEndedDelegate(FortMinigameProgress)
  ;[0 0.8698 0.7627 0.25] [SERVER/CLIENT] Handle minigame events
  (Variables|Gameplay|SetbIsMinigameInProgress false)
  (CallFunction|UpdateButtonCollision)
  (Networking|CanExecuteCosmeticEvents
    (:True
      (CallFunction|UpdateNextItemToAdd)
      (Visuals|UpdateColors)
      (CallFunction|UpdatePreviewVisibility)
      (Default|CallOnGameStateChanged (Variables|Gameplay|GetbIsMinigameInProgress))
      (Class|BPcLockHologramV2|UpdateScreenColorfromLocalPawn (Variables|Default|GetHologram%20Array)) ; Need to update screen colors for holograms\nat beginning of minigame as screen colors use\nin progress state as part of their logic.
    )
    (:False
      (Utilities|HandleDeviceReset))))

(event Custom|On%20Play%20Mode%20Changed
  (params (Minigame :type Fort%20Minigame%20Object%20Reference) (bIsInPlayMode :type Boolean))
  ;[0 0.8698 0.7627 0.25] [SERVER/CLIENT] Handle minigame events
  (Variables|Actor|SetbCanBeDamaged (not bIsInPlayMode)) ; 'Damage' during edit mode to allow\nhit to swap active key item slot.
  (Default|CallOnPlayModeChanged bIsInPlayMode)
  (Networking|CanExecuteCosmeticEvents
    (:True
      (CallFunction|OnDeviceReset))))

(event EventOnHitByMusicSequencer
  (params (InstigatedBy :type Controller%20Object%20Reference))
  ;[0.323 0 1 0.25] [SERVER ONLY] Transmitters  / Receivers
  (if (Variables|UserOptions|GetActivatedBySequencers)
    (CallFunction|ActivateRemotely InstigatedBy)))

(event Custom|HandleActivate
  (params (Instigator :type Controller%20Object%20Reference))
  ;[0.323 0 1 0.25] [SERVER ONLY] Transmitters  / Receivers
  (CallFunction|ActivateRemotely Instigator))

(event Custom|Activate%20Remotely
  (params (Controller :type Controller%20Object%20Reference))
  (bind _returnvalue (Pawn|GetControlledPawn Controller))
  ;[0.323 0 1 0.25] [SERVER ONLY] Transmitters  / Receivers
  (Utilities|IsValid Controller
    (:Is%20Valid
      (Utilities|IsValid _returnvalue
        (:Is%20Valid
          (if (TestingRequirements|Team&ClassMatch _returnvalue)
            (CallFunction|InteractIfHasKeys _returnvalue)))))))

(event EventDestroyed
  ;[0.9375 0 0] Destruction - Unbind and stop any active audio.
  (Initialization|BindEvents false)
  (bind _device_active_loop (Variables|Default|GetDevice%20Active%20Loop)
    (:then
      (Audio|Components|Audio|Stop _device_active_loop))
    (:else)))

(event OnVisibilityChanged(VisibleInGameComponent_StaticMesh)
  (params (Visible :type Boolean))
  ;[0 0.8698 0.7627 0.25] [SERVER/CLIENT] Updating Collision on Visibility Change
  (CallFunction|UpdateButtonCollision))

(event OnKeysConsumed(BP_KeyLockComponent)
  (params (NumKeysConsumed :type Integer))
  ;[1 1 1] Number Of Keys Consumed Delegates
  (Variables|Gameplay|SetScoreToAdd NumKeysConsumed))

(event OnKeysConsumed(BP_KeyLockComponent2)
  (params (NumKeysConsumed :type Integer))
  ;[1 1 1] Number Of Keys Consumed Delegates
  (Variables|Gameplay|SetScoreToAdd1 NumKeysConsumed))

(event OnKeysConsumed(BP_KeyLockComponent3)
  (params (NumKeysConsumed :type Integer))
  ;[1 1 1] Number Of Keys Consumed Delegates
  (Variables|Gameplay|SetScoreToAdd2 NumKeysConsumed))

(event OnKeyDefinitionUpdated(BP_KeyLockComponent)
  (params (NewKeyDefinition :type Item%20Variant%20Handle%20Structure))
  ;[0 0.9032 1 0.1] Track what items are held in each keylock component.
  (Utilities|Array|SetArrayElem (Variables|Default|GetTrackedItems) 0 NewKeyDefinition true))

(event OnKeyDefinitionUpdated(BP_KeyLockComponent2)
  (params (NewKeyDefinition :type Item%20Variant%20Handle%20Structure))
  ;[0 0.9032 1 0.1] Track what items are held in each keylock component.
  (Utilities|Array|SetArrayElem (Variables|Default|GetTrackedItems) 1 NewKeyDefinition true))

(event OnKeyDefinitionUpdated(BP_KeyLockComponent3)
  (params (NewKeyDefinition :type Item%20Variant%20Handle%20Structure))
  ;[0 0.9032 1 0.1] Track what items are held in each keylock component.
  (Utilities|Array|SetArrayElem (Variables|Default|GetTrackedItems) 2 NewKeyDefinition true))

(event OnComponentBeginOverlap(ShowPreviewArea)
  (params (OverlappedComponent :type Primitive%20Component%20Object%20Reference) (OtherActor :type Actor%20Object%20Reference) (OtherComp :type Primitive%20Component%20Object%20Reference) (OtherBodyIndex :type Integer) (bFromSweep :type Boolean) (SweepResult :type Hit%20Result%20Structure%20%28by%20ref%29))
  ;[1 0.6509 0 0.25] [LOCAL CLIENT ONLY] Set to show or hide the required preview
  (CallFunction|PawnOverlapUpdated (TestingRequirements|CanLocalTeam&ClassInteract) OtherActor))

(event Custom|Pawn%20Overlap%20Updated
  (params (Valid :type Boolean) (Pawn%20Object :type Object%20Reference))
  ;[1 0.6509 0 0.25] [LOCAL CLIENT ONLY] Set to show or hide the required preview
  (if (or (Variables|Gameplay|GetbIsEnabled) (not (Variables|Gameplay|GetbIsMinigameInProgress)))
    (Networking|CanExecuteCosmeticEvents
      (:True
        (bind _aspawn (Utilities|Casting|CastToPawn Pawn%20Object))
        (if Valid
          (if (Pawn|IsLocallyControlled _aspawn)
            (CallFunction|HandleLocalPawnEntertoPreviewArea _aspawn)
            (elif (Replay|GetIsPlayingReplay)
              _))
          (CallFunction|ShowKeyPreview (TestingRequirements|TestRequirements _aspawn))
          (elif (Pawn|IsLocallyControlled _aspawn)
            (CallFunction|HandleLocalPawnExitPreviewArea)
            (CallFunction|HideKeyPreview)
            (elif (Replay|GetIsPlayingReplay)
              _)))))))

(event OnComponentEndOverlap(ShowPreviewArea)
  (params (OverlappedComponent :type Primitive%20Component%20Object%20Reference) (OtherActor :type Actor%20Object%20Reference) (OtherComp :type Primitive%20Component%20Object%20Reference) (OtherBodyIndex :type Integer))
  ;[1 0.6509 0 0.25] [LOCAL CLIENT ONLY] Set to show or hide the required preview
  (CallFunction|PawnOverlapUpdated false OtherActor))

(event OnKeyUnlockComplete(LockHologram_11)
  (bind _resetdelay (Variables|UserOptions|GetResetDelay))
  ;[1 0.6509 0 0.25] [CLIENT ONLY] Play the unlock feedback in the device
  (if (not (TestingRequirements|AnyLocked))
    (if (Variables|UserOptions|GetDisableAfterUse)
      (Utilities|FlowControl|Delay _resetdelay)
      (Visuals|UpdateDeviceDisplayState "NewEnumerator1")
      (else
        (Utilities|FlowControl|Delay _resetdelay)
        (CallFunction|HideKeyPreview)))
    (CallFunction|HandleLocalPawnInventoryChanged)))

(event OnUnlocked(BP_KeyLockComponent)
  ;[1 0.6509 0 0.25] [CLIENT ONLY] Play the unlock feedback in the device
  (Networking|CanExecuteCosmeticEvents
    (:True
      (if (not (TestingRequirements|AnyLocked))
        (Visuals|UpdateDeviceDisplayState "NewEnumerator0")))))

(event OnKeyUnlockComplete(LockHologram_22)
  (bind _resetdelay (Variables|UserOptions|GetResetDelay))
  ;[1 0.6509 0 0.25] [CLIENT ONLY] Play the unlock feedback in the device
  (if (not (TestingRequirements|AnyLocked))
    (if (Variables|UserOptions|GetDisableAfterUse)
      (Utilities|FlowControl|Delay _resetdelay)
      (Visuals|UpdateDeviceDisplayState "NewEnumerator1")
      (else
        (Utilities|FlowControl|Delay _resetdelay)
        (CallFunction|HideKeyPreview)))
    (CallFunction|HandleLocalPawnInventoryChanged)))

(event OnUnlocked(BP_KeyLockComponent2)
  ;[1 0.6509 0 0.25] [CLIENT ONLY] Play the unlock feedback in the device
  (Networking|CanExecuteCosmeticEvents
    (:True
      (if (not (TestingRequirements|AnyLocked))
        (Visuals|UpdateDeviceDisplayState "NewEnumerator0")))))

(event OnKeyUnlockComplete(LockHologram_33)
  (bind _resetdelay (Variables|UserOptions|GetResetDelay))
  ;[1 0.6509 0 0.25] [CLIENT ONLY] Play the unlock feedback in the device
  (if (not (TestingRequirements|AnyLocked))
    (if (Variables|UserOptions|GetDisableAfterUse)
      (Utilities|FlowControl|Delay _resetdelay)
      (Visuals|UpdateDeviceDisplayState "NewEnumerator1")
      (else
        (Utilities|FlowControl|Delay _resetdelay)
        (CallFunction|HideKeyPreview)))
    (CallFunction|HandleLocalPawnInventoryChanged)))

(event OnUnlocked(BP_KeyLockComponent3)
  ;[1 0.6509 0 0.25] [CLIENT ONLY] Play the unlock feedback in the device
  (Networking|CanExecuteCosmeticEvents
    (:True
      (if (not (TestingRequirements|AnyLocked))
        (Visuals|UpdateDeviceDisplayState "NewEnumerator0")))))

(event Custom|OnDeviceReset
  (bind (_in_range _can_interact _color) (CallFunction|GetLocalPlayerDetails))
  ;[1 0.6509 0 0.25] [CLIENT ONLY] Reset FX
  (Visuals|UpdateDeviceDisplayIcon false)
  (if _in_range
    (CallFunction|ShowKeyPreview _can_interact)
    (else
      (CallFunction|HideKeyPreview)
      (Visuals|UpdatePreviewColors false _color)))
  (CallFunction|HandleLocalPawnInventoryChanged)
  (Visuals|UpdateDeviceDisplayState "NewEnumerator1"))

(event Custom|ShowKeyPreview
  (params (CanInteract :type Boolean))
  ;[1 0.6509 0 0.25] [CLIENT ONLY] Show or Hide the Preview
  (if (not (Variables|Gameplay|GetbIsPreviewOn))
    (Visuals|SetupPreviewPreAnimation CanInteract)
    (if (Visibility|ShouldbeVisible (Variables|Default|GetVisibleInGameComponent_Hologram1))
      (Audio|PlaySoundatLocation (Variables|Audio|GetShow%20Display%20Sound) (Transformation|GetActorLocation))
      (bind (_direction _on) (Timeline|Show%20Preview PlayFromStart)
        (:Update
          (Visuals|TickPreviewAnimation _on))
        (:Finished
          (Visuals|TickPreviewAnimation 1.0)
          (Variables|Gameplay|SetbIsPreviewOn true)
          (Audio|Components|Audio|Play (Variables|Default|GetDevice%20Active%20Loop) 0.0))))))

(event Custom|HideKeyPreview
  ;[1 0.6509 0 0.25] [CLIENT ONLY] Show or Hide the Preview
  (if (Variables|Gameplay|GetbIsPreviewOn)
    (Components|Timeline|Stop (Variables|Device_ConditionalButton_V2|GetShow%20Preview))
    (if (Visibility|ShouldbeVisible (Variables|Default|GetVisibleInGameComponent_Hologram1))
      (Audio|PlaySoundatLocation (Variables|Audio|GetHide%20Display%20Sound) (Transformation|GetActorLocation)))
    (bind (_direction _off) (Timeline|Hide%20Preview PlayFromStart)
      (:Update
        (Visuals|TickPreviewAnimation _off))
      (:Finished
        (Visuals|TickPreviewAnimation 0.0)
        (Variables|Gameplay|SetbIsPreviewOn false)
        (Audio|Components|Audio|Stop (Variables|Default|GetDevice%20Active%20Loop))))))

(event Custom|OnItemCaptured
  (params (capturedPickup :type Fort%20Pickup%20Object%20Reference))
  (bind _outactors (Actor|GetAttachedActors capturedPickup true false))
  ;[1 0.6509 0 0.25] [CLIENT ONLY][Item Captured FX]
  (Utilities|FlowControl|Delay 1.5) ; This a workaround, because the pickup mesh doesn't spawns immediately
  (Utilities|IsValid capturedPickup
    (:Is%20Valid
      (Networking|CanExecuteCosmeticEvents
        (:True
          (Variables|CurrentPickupData|SetPickedUpLocation (Transformation|GetActorLocation capturedPickup))
          (if (> (Utilities|Array|Length _outactors) 0)
            (bind _asb_pickups_parent (Utilities|Casting|CastToB_Pickups_Parent (Utilities|Array|Get(acopy) _outactors)))
            (Variables|CurrentPickupData|SetItemMIDArray (Variables|/Script/FortniteGame.FortPickupsParent|GetMIDs _asb_pickups_parent))
            (bind (_direction _movemeshtocenter _suckeffect) (Timeline|DragItem PlayFromStart)
              (:Update
                (Utilities|IsValid _asb_pickups_parent
                  (:Is%20Valid
                    (Transformation|SetActorLocation _asb_pickups_parent (Math|Vector|Lerp(Vector) (Variables|CurrentPickupData|GetPickedUpLocation) (Transformation|GetWorldLocation (Variables|Debug|GetStaticMeshComponent)) _movemeshtocenter) false false)
                    (for _array_element (Variables|CurrentPickupData|GetItemMIDArray)
                      (Rendering|Material|SetVectorParameterValue _array_element "WorldPostionOfBlackHole" (Math|Color|MakeColor (.x (Transformation|GetWorldLocation (Variables|Debug|GetStaticMeshComponent))) (.y (Transformation|GetWorldLocation (Variables|Debug|GetStaticMeshComponent))) (.z (Transformation|GetWorldLocation (Variables|Debug|GetStaticMeshComponent))) _suckeffect) :!class /Script/Engine.MaterialInstanceDynamic) ; Also Sets Animation
                    ))))
              (:Finished
                (Utilities|IsValid _asb_pickups_parent
                  (:Is%20Valid
                    (Rendering|SetActorHiddenInGame _asb_pickups_parent true)))))))))))

(fn ActivateFromOtherDevice
  (params (Instigator :type Controller%20Object%20Reference) (Requester :type ERequesterType%20Enum :default "NewEnumerator0"))
  (return false))

(fn AgentHasItems
  (params (Controller :type Controller%20Object%20Reference))
  (bind _returnvalue (Pawn|GetControlledPawn Controller))
  (Utilities|IsValid Controller
    (:Is%20Valid
      (Utilities|IsValid _returnvalue
        (:Is%20Valid
          (return (TestingRequirements|PawnHasKeys _returnvalue)))))
    (:Is%20Not%20Valid
      (return false))))

(fn AgentHoldingItem
  (params (Controller :type Controller%20Object%20Reference))
  (bind _returnvalue (Pawn|GetControlledPawn Controller))
  (Utilities|IsValid Controller
    (:Is%20Valid
      (Utilities|IsValid _returnvalue
        (:Is%20Valid
          (return (TestingRequirements|PawnIsHoldingItem _returnvalue -1 true)))
        (:Is%20Not%20Valid
          _)))
    (:Is%20Not%20Valid
      _))
  (return false))

(fn AgentHoldingItemAtIndex
  (params (Controller :type Controller%20Object%20Reference) (KeyIndex :type Integer))
  (bind _returnvalue (Pawn|GetControlledPawn Controller))
  (Utilities|IsValid Controller
    (:Is%20Valid
      (Utilities|IsValid _returnvalue
        (:Is%20Valid
          (return (TestingRequirements|PawnIsHoldingItem _returnvalue KeyIndex true)))
        (:Is%20Not%20Valid
          _)))
    (:Is%20Not%20Valid
      _))
  (return false))

(fn AnyLocked
  (for _index (range (Variables|UserOptions|GetNumberOfKeyItemSlots))
    (if (Class|FortCreativeKeyLockComponent|IsLocked (Utilities|Array|Get(acopy) (Variables|Default|GetLock%20Array) _index))
      (return true)))
  (return false))

(fn AreThereLocalPlayersinthePreviewArea
  (Networking|CanExecuteCosmeticEvents
    (:True
      (Utilities|FlowControl|Sequence
        (:then_0
          (bind _array_element (Utilities|Array|ForEachLoopwithBreak (Collision|GetOverlappingActors (Variables|Default|GetShowPreviewArea) "/Script/Engine.Pawn" :!class /Script/Engine.PrimitiveComponent))
            (:LoopBody
              (bind _aspawn (Utilities|Casting|CastToPawn _array_element))
              (if (or (Pawn|IsLocallyControlled _aspawn) (Replay|GetIsPlayingReplay))
                (LocalVariables|SetbPlayersFound true)))))
        (:then_1
          (return (LocalVariables|GetbPlayersFound)))))))

(fn AssigntoRegistry
  (CreativeObjectRegistry|GetOrCreateObjectRegister self "(TagName=\"Creative.Register.ItemSources.Requires\")" true 0 (Variables|Default|GetLock%20Register)))

(fn BindEvents
  (params (Bind :type Boolean))
  (bind _returnvalue (Fortnite|GetFirstLocalFortPlayerController true))
  (bind _outputdelegate (EventDispatchers|CreateEvent :!event Local%20Player%20Updated))
  (bind _outputdelegate_2 (EventDispatchers|CreateEvent :!event OnWorldInventoryChanged))
  (bind _outputdelegate_3 (EventDispatchers|CreateEvent :!event On%20Play%20Mode%20Changed))
  (bind _outputdelegate_4 (EventDispatchers|CreateEvent :!event Local%20Player%20Updated))
  (bind _minigame (LocalVariables|GetMinigame))
  (bind _minigame_2 (LocalVariables|GetMinigame))
  (Networking|CanExecuteCosmeticEvents
    (:True
      (bind _fortminigameprogress (Variables|Default|GetFortMinigameProgress)
        (:then
          (bind _currentminigame (Variables|/Script/FortniteGame.FortMinigameLogicComponent|GetCurrentMinigame _fortminigameprogress)
            (:then
              (LocalVariables|SetMinigame _currentminigame)
              (if (Variables|GetBind)
                ;[1 0.5253 0 0.2] Team/Class Colors and visibility require us to bind to events to see when players change colors or teams change their colors.
                ;
                (bind _asfort_player_controller_athena (Utilities|Casting|CastToFortPlayerControllerAthena _returnvalue))
                (EventDispatchers|BindEventtoOnWorldInventoryChangedDelegate _asfort_player_controller_athena _outputdelegate_2)
                (FortPlayerControllerAthena|BindEventtoOnLocalPlayerChangedTeams _asfort_player_controller_athena _outputdelegate_4)
                (Minigame|BindEventtoMinigameTeamConfigUpdated _minigame _outputdelegate_4)
                (Minigame|BindEventtoOnPlayModeChangedDelegate _minigame _outputdelegate_3)
                (Minigame|BindEventtoOnAnyLocalPlayerChangedClass _minigame _outputdelegate)
                (else
                  (bind _asfort_player_controller_athena_2 (Utilities|Casting|CastToFortPlayerControllerAthena _returnvalue))
                  (EventDispatchers|UnbindEventfromOnWorldInventoryChangedDelegate _asfort_player_controller_athena_2 _outputdelegate_2)
                  (FortPlayerControllerAthena|BindEventtoOnLocalPlayerChangedTeams _asfort_player_controller_athena_2 _outputdelegate_4)
                  (Minigame|UnbindEventfromMinigameTeamConfigUpdated _minigame_2 _outputdelegate_4)
                  (Minigame|UnbindEventfromOnPlayModeChangedDelegate _minigame_2 _outputdelegate_3)
                  (Minigame|UnbindEventfromOnAnyLocalPlayerChangedClass _minigame_2 _outputdelegate))))
            (:else
              (Minigame|BindEventtoOnAddedtoMinigame _fortminigameprogress (EventDispatchers|CreateEvent :!event On%20Retry%20Bind%20Delegates)))))
        (:else)))))

(fn BlueprintCanInteract
  (params (InteractingPawn :type Fort%20Pawn%20Object%20Reference) (InteractionBeingAttempted :type EInteractionBeingAttempted%20Enum :default "FirstInteraction") (InteractionType :type TInteractionType%20Enum :default "IT_NoInteraction"))
  (Utilities|FlowControl|Sequence
    (:then_0
      (if (Variables|Gameplay|GetbIsMinigameInProgress)
        (if (and (not (Variables|Gameplay|GetCoolingDown)) (Variables|Gameplay|GetbIsEnabled))
          (LocalVariables|SetInteractionAllowed (Math|Boolean|ANDBoolean (Variables|UserOptions|Basic|GetAllowInteraction) (>= (Variables|UserOptions|Basic|GetInteractTime) 0.0) (TestingRequirements|Team&ClassMatch (Variables|GetInteractingPawn)))))
        (else
          (Utilities|IsValid (Variables|/Script/FortniteGame.FortActorOptionsComponent|GetInternal_PlayerOptions (Variables|Default|GetToyOptionsComponent))
            (:Is%20Valid
              (bind _interactingpawn (Variables|GetInteractingPawn)
                (:then
                  (bind _asfort_player_controller_athena (Utilities|Casting|CastToFortPlayerControllerAthena (Pawn|GetController _interactingpawn)))
                  (LocalVariables|SetInteractionAllowed (Creative|CanPlayerCreateinCurrentExperience _asfort_player_controller_athena)))
                (:else)))))))
    (:then_1
      (return (LocalVariables|GetInteractionAllowed)))))

(fn BlueprintGetFailedInteractionString
  (params (InteractingPawn :type Fort%20Pawn%20Object%20Reference) (InteractionBeingAttempted :type EInteractionBeingAttempted%20Enum :default "FirstInteraction"))
  (bind _interactingpawn (Variables|GetInteractingPawn)
    (:then
      (LocalVariables|SetCanTeamInteract (TestingRequirements|CanTeamInteract _interactingpawn))
      (LocalVariables|SetCanClassInteract (TestingRequirements|CanClassInteract _interactingpawn))
      (if (Math|Boolean|NORBoolean (LocalVariables|GetCanTeamInteract) (LocalVariables|GetCanClassInteract))
        (return (Utilities|Text|FormatText (Variables|UserOptions|GetInvalidTeam/ClassText) (Utilities|Text|MakeLiteralText "NSLOCTEXT(\"[182ED5CAF3AB930F56681381A158961B]\", \"1ABC25E84EABC8E4434511BCA240B862\", \"Invalid Team & Class\")")))
        (else
          (return (Utilities|Text|FormatText (Variables|UserOptions|GetInvalidTeam/ClassText) (select (LocalVariables|GetCanClassInteract) (select (LocalVariables|GetCanTeamInteract) "" "NSLOCTEXT(\"[182ED5CAF3AB930F56681381A158961B]\", \"6BE0D931457C5FE14F2219907FA7CCDD\", \"Invalid Team\")") "NSLOCTEXT(\"[182ED5CAF3AB930F56681381A158961B]\", \"5A988A8A472645D0E1F4D7BB30344923\", \"Invalid Class\")"))))))
    (:else)))

(fn BlueprintGetInteractionString
  (params (InteractingPawn :type Fort%20Pawn%20Object%20Reference) (InteractionBeingAttempted :type EInteractionBeingAttempted%20Enum :default "FirstInteraction"))
  (if (Variables|Gameplay|GetbIsMinigameInProgress)
    (return (CallFunction|FormatInteractText (select (TestingRequirements|PawnHasKeys InteractingPawn) (select (TestingRequirements|PawnIsHoldingItem InteractingPawn) (select (Utilities|Text|TextIsEmpty (select (Utilities|Text|TextIsEmpty (Variables|Verse|GetInteractionStringOverride)) (Variables|UserOptions|GetInteractText) (Variables|Verse|GetInteractionStringOverride))) (Variables|Default|GetInteractString_Default) (select (Utilities|Text|TextIsEmpty (Variables|Verse|GetInteractionStringOverride)) (Variables|UserOptions|GetInteractText) (Variables|Verse|GetInteractionStringOverride))) (Variables|Default|GetNot%20Holding%20Item%20Text)) (select (Utilities|Text|TextIsEmpty (Variables|UserOptions|GetMissingItemsText)) (Variables|Default|GetMissingItemsText_Default) (Variables|UserOptions|GetMissingItemsText)))))
    (else
      (bind _returnvalue (|CallParentFunction InteractingPawn))
      (return _returnvalue))))

(fn BlueprintGetInteractionTime
  (params (InteractingPawn :type Fort%20Pawn%20Object%20Reference) (InteractionBeingAttempted :type EInteractionBeingAttempted%20Enum :default "FirstInteraction"))
  (return (Math|Float|Max(Float) (Math|Float|SelectFloat (Variables|Verse|GetInteractTimeOverride) (Variables|UserOptions|Basic|GetInteractTime) (>= (Variables|Verse|GetInteractTimeOverride) 0))) (and (Variables|Gameplay|GetbIsMinigameInProgress) (TestingRequirements|PawnHasKeys InteractingPawn))))

(fn BlueprintModifyIncomingDamage
  (params (Damage :type Float%20%28single-precision%29) (InTags :type Gameplay%20Tag%20Container%20Structure%20%28by%20ref%29) (EffectContext :type Gameplay%20Effect%20Context%20Handle%20Structure%20%28by%20ref%29) (EventInstigator :type Controller%20Object%20Reference) (DamageCauser :type Actor%20Object%20Reference))
  ;[1 1 1] Don't do damage, but we want to switch item when it takes a hit.
  (if (not (Class|CreativeCommonDeviceFunctionLibrary|IsLogicinPlayMode (Variables|Default|GetFortMinigameProgress)))
    (CallFunction|SwitchItem)
    (else
      _))
  (return 0.0))

(fn BlueprintOnLocalInteract
  (params (InteractingPawn :type Fort%20Player%20Pawn%20Object%20Reference) (InteractionBeingAttempted :type EInteractionBeingAttempted%20Enum :default "FirstInteraction"))
  (Utilities|FlowControl|Sequence
    (:then_0
      (LocalVariables|SetbConsumeInteract false))
    (:then_1
      (if (Variables|Gameplay|GetbIsMinigameInProgress)
        (if (not (TestingRequirements|PawnHasKeys (Variables|GetInteractingPawn)))
          (FailedAction|UnabletoPerformAction (Variables|GetInteractingPawn) true (GameplayTags|MakeGameplayTagContainerfromTag "(TagName=\"Activate.Failed.Tags.Blocked\")"))
          (for _array_element (Variables|Default|GetHologram%20Array)
            (bind _returnvalue (Class|FortCreativeKeyLockComponent|BlueprintHasKeystoInteract (Variables|/CreativeCoreDevices/ConditionalButton/BPc_LockHologram_v2.BPc_LockHologram_v2_C|GetKey%20Lock _array_element) (Variables|GetInteractingPawn)))
            (PlayAnimations|PlayInteraction _array_element (or _returnvalue (not (Class|FortCreativeKeyLockComponent|IsLocked (Variables|/CreativeCoreDevices/ConditionalButton/BPc_LockHologram_v2.BPc_LockHologram_v2_C|GetKey%20Lock _array_element))))))
          (Visuals|UpdatePreviewColors false)
          (elif (Variables|UserOptions|GetbRequireHoldingItem)
            (if (not (TestingRequirements|PawnIsHoldingItem (Variables|GetInteractingPawn)))
              (FailedAction|UnabletoPerformAction (Variables|GetInteractingPawn) true (GameplayTags|MakeGameplayTagContainerfromTag "(TagName=\"Activate.Failed.Tags.Blocked\")"))
              (for _array_element_2 (Variables|Default|GetHologram%20Array)
                (PlayAnimations|PlayInteraction _array_element_2 false)))))
        (else
          (bind _returnvalue_2 (Interaction|OnLocalInteract (Variables|Default|GetToyOptionsComponent) (Variables|GetInteractingPawn)))
          (LocalVariables|SetbConsumeInteract _returnvalue_2))))
    (:then_2
      (return (LocalVariables|GetbConsumeInteract)))))

(fn CanLocalTeam&ClassInteract
  (bind _returnvalue (Fortnite|GetFirstLocalFortPlayerController true))
  (bind _fortminigameprogress (Variables|Default|GetFortMinigameProgress)
    (:then
      (bind _currentminigame (Variables|/Script/FortniteGame.FortMinigameLogicComponent|GetCurrentMinigame _fortminigameprogress)
        (:then
          (if (Minigame|IsinPlayMode _currentminigame :!class /Script/FortniteGame.FortMinigame)
            (Networking|CanExecuteCosmeticEvents
              (:True
                (Utilities|IsValid _returnvalue
                  (:Is%20Valid
                    (return (and (TestingRequirements|CanTeamInteract (Game|Player|GetPlayerPawn _returnvalue)) (TestingRequirements|CanClassInteract (Game|Player|GetPlayerPawn _returnvalue)))))
                  (:Is%20Not%20Valid
                    _)))
              (:False
                _))
            (return true)
            (else
              _)))
        (:else
          _)))
    (:else
      _)))

(fn CanClassInteract
  (params (Actor :type Actor%20Object%20Reference))
  (if (Creative|ClassOption|IsClassType (Variables|UserOptions|GetAllowedClass) "Any")
    (return true)
    (else
      (bind _asfort_pawn (Utilities|Casting|CastToFortPawn Actor)
        (:then
          (bind _playerstate (Variables|/Script/Engine.Pawn|GetPlayerState _asfort_pawn)
            (:then
              (bind _asfort_player_state (Utilities|Casting|CastToFortPlayerState _playerstate)
                (:then
                  (bind _fortminigameprogress (Variables|Default|GetFortMinigameProgress)
                    (:then
                      (bind _currentminigame (Variables|/Script/FortniteGame.FortMinigameLogicComponent|GetCurrentMinigame _fortminigameprogress)
                        (:then
                          (Utilities|IsValid (Minigame|GetFNEClassComponent _currentminigame)
                            (:Is%20Valid
                              (return (== (Minigame|GetClassSlotforPlayer (Minigame|GetFNEClassComponent _currentminigame) _asfort_player_state false) (Creative|ClassOption|ConverttoClassSlotIndex (Variables|UserOptions|GetAllowedClass)))))
                            (:Is%20Not%20Valid
                              _)))
                        (:else
                          _))
                      (return false))
                    (:else
                      _)))
                (:CastFailed
                  _)))
            (:else
              _)))
        (:CastFailed
          _)))))

(fn CanInteract
  (params (InteractingPawn :type Fort%20Pawn%20Object%20Reference) (InteractionBeingAttempted :type EInteractionBeingAttempted%20Enum :default "FirstInteraction") (InteractionType :type TInteractionType%20Enum :default "IT_NoInteraction"))
  (return false))

(fn CanTeamInteract
  (params (Actor :type Actor%20Object%20Reference))
  (bind _activatingteam (Variables|UserOptions|GetActivatingTeam))
  (bind _returnvalue (Creative|TeamOption|IsTeamType _activatingteam "Any"))
  (bind _returnvalue_2 (Creative|TeamOption|IsTeamType _activatingteam))
  (bind _returnvalue_3 (Team|GetActorTeam Actor))
  (bind _returnvalue_4 (Creative|TeamOption|ConverttoTeam _activatingteam))
  (bind _returnvalue_5 (Math|Conversions|ToInteger(Byte) _returnvalue_4))
  (bind _returnvalue_6 (== _returnvalue_5 _returnvalue_3))
  (bind _returnvalue_7 (and _returnvalue_2 _returnvalue_6))
  (bind _returnvalue_8 (or _returnvalue _returnvalue_7))
  (Utilities|IsValid Actor
    (:Is%20Valid
      (Utilities|FlowControl|Sequence
        (:then_0
          (if true
            (switch Utilities|FlowControl|Switch|SwitchonELockDevice_TeamAllowed (Variables|UserOptions|GetCanBeUsedBy)
              (:NewEnumerator0
                (LocalVariables|SetInteractAllowed _returnvalue_8))
              (:NewEnumerator1
                (LocalVariables|SetInteractAllowed (not _returnvalue_8)))
              (:NewEnumerator2
                (LocalVariables|SetInteractAllowed true)))
            (else
              (LocalVariables|SetInteractAllowed true))))
        (:then_1
          (return (LocalVariables|GetInteractAllowed)))))))

(fn CaptureItem
  (params (Item :type Actor%20Object%20Reference) (ComponentIndex :type Integer))
  (if (not (Variables|Gameplay|GetbIsMinigameInProgress))
    (bind _asfort_pickup_athena (Utilities|Casting|CastToFortPickupAthena Item))
    (Utilities|IsPickupValid _asfort_pickup_athena (Variables|Default|GetInvalidPickupGameplayTagContainer)
      (:IsValid
        (if (and (not (Pickup|PickedUp _asfort_pickup_athena)) (not (Utilities|Equal(Object) _asfort_pickup_athena (Variables|Gameplay|GetLastItemCaptured))))
          (Pickup|SetPickedUp _asfort_pickup_athena true)
          (Actor|DestroyActor (Variables|Gameplay|GetLastItemCaptured))
          (Variables|Gameplay|SetLastItemCaptured _asfort_pickup_athena)
          (Variables|CurrentPickupData|SetCurrentItemCaptured _asfort_pickup_athena)
          (Class|FortCreativeKeyLockComponent|SetKeyItemDefinition (Utilities|Select :Index (Variables|GetComponentIndex) :Option%200 (Variables|Default|GetBP_KeyLockComponent) :Option%201 (Variables|Default|GetBP_KeyLockComponent2) :Option%202 (Variables|Default|GetBP_KeyLockComponent3)) (Creative|ItemVariant|ToItemVariantHandle(ItemEntry) (Pickup|GetItemEntry _asfort_pickup_athena 0)))
          (Variables|SetBy-RefVar (Utilities|Select :Index (Variables|GetComponentIndex) :Option%200 (Variables|Default|GetStackSize_1) :Option%201 (Variables|Default|GetStackSize_2) :Option%202 (Variables|Default|GetStackSize_3)) (Math|Integer|Min(Integer) (.Count (Class|FortItemFunctionLibrary|BreakFortItemEntry (Pickup|GetItemEntry _asfort_pickup_athena 0))) (Variables|Default|GetStackSizeMax)))
          (Actor|SetLifeSpan _asfort_pickup_athena 3.0) ; We give the client time so he can play the corresponding feedback
          (Audio|PlaySoundatLocation (Variables|Audio|GetAccept%20Item%20Sound) (Transformation|GetActorLocation))
          (Initialization|SetKeysRequired)
          (Default|IncrementNextItemToAdd))))))

(fn DeviceReset
  (Networking|SwitchHasAuthority
    (:Authority
      (Utilities|FlowControl|Sequence
        (:then_0
          (Variables|Verse|SetInteractTimeOverride -1.0)
          (Class|FortLocalizableMessageComponent|LocalizeandSetMessage (Variables|Default|GetFortLocalizableMessage) (Utilities|Struct|MakeLocalizableMessage) "InteractionStringOverride" "InteractText")
          (Variables|Verse|SetScoreOnKeyItem1%20Consumed%20Override -1)
          (Variables|Verse|SetScoreOnKeyItem2%20Consumed%20Override -1)
          (Variables|Verse|SetScoreOnKeyItem3%20Consumed%20Override -1))
        (:then_1
          (Class|FortCreativeKeyLockComponent|Reset (Variables|Default|GetLock%20Array) false)
          (CallFunction|SetEnableState (Variables|UserOptions|GetEnabledOnGameStart))
          (Variables|Gameplay|SetIconsWereToggled false)
          (Utilities|Array|Clear (Variables|Default|GetCompletedKeyIndexes)))))))

(fn FormatInteractText
  (params (Input :type Text))
  (bind _showitemrarityonmissingitemlist (Variables|UserOptions|GetShowItemRarityOnMissingItemList))
  (bind _hologram_array (Variables|Default|GetHologram%20Array))
  (for _index (range (Variables|UserOptions|GetNumberOfKeyItemSlots))
    (if (Class|BPcLockHologramV2|GetMissingText (Utilities|Array|Get(acopy) _hologram_array _index) _showitemrarityonmissingitemlist)
      (LocalVariables|SetItems (Utilities|Text|FormatText "NSLOCTEXT(\"[182ED5CAF3AB930F56681381A158961B]\", \"BA95313242FA43F7283242BC747FBA6F\", \"{items}\\r\\n{new}\")" (LocalVariables|GetItems) (.Missing (Class|BPcLockHologramV2|GetMissingText (Utilities|Array|Get(acopy) _hologram_array _index) _showitemrarityonmissingitemlist))))))
  (return (Utilities|Text|FormatText Input (LocalVariables|GetItems))))

(fn GetDeviceColor
  (bind _icon (CallFunction|ValidToColor "NewEnumerator1"))
  (return _icon))

(fn GetInstigatingController
  (return 0))

(fn GetLocalPlayerDetails
  (bind _returnvalue (Fortnite|GetFirstLocalFortPlayerController true))
  (bind _returnvalue_2 (Game|Player|GetPlayerPawn _returnvalue))
  (Networking|CanExecuteCosmeticEvents
    (:True
      (Utilities|IsValid _returnvalue
        (:Is%20Valid
          (Utilities|IsValid _returnvalue_2
            (:Is%20Valid
              (return (Math|Boolean|ANDBoolean (TestingRequirements|AreThereLocalPlayersInThePreviewArea) (or (Variables|Gameplay|GetbIsEnabled) (not (Variables|Gameplay|GetbIsMinigameInProgress))) (TestingRequirements|CanLocalTeam&ClassInteract)) (TestingRequirements|TestRequirements _returnvalue_2) (select (TestingRequirements|TestRequirements _returnvalue_2) "NewEnumerator1" "NewEnumerator0")))
            (:Is%20Not%20Valid
              _)))
        (:Is%20Not%20Valid
          _))
      (return false false "NewEnumerator0"))))

(fn GetCollisionStaticMesh
  (return "0, 0, 0" "/Game/Creative/Devices/Lock_and_Key/Meshes/Lock_and_Key.Lock_and_Key"))

(fn GetInteractionText
  (return (Utilities|String|ToString(Text) (select (Utilities|Text|TextIsEmpty (select (Utilities|Text|TextIsEmpty (Variables|Verse|GetInteractionStringOverride)) (Variables|UserOptions|GetInteractText) (Variables|Verse|GetInteractionStringOverride))) (Variables|Default|GetInteractString_Default) (select (Utilities|Text|TextIsEmpty (Variables|Verse|GetInteractionStringOverride)) (Variables|UserOptions|GetInteractText) (Variables|Verse|GetInteractionStringOverride))))))

(fn GetInteractionTime
  (return (Math|Float|Max(Float) (Math|Float|SelectFloat (Variables|Verse|GetInteractTimeOverride) (Variables|UserOptions|Basic|GetInteractTime) (>= (Variables|Verse|GetInteractTimeOverride) 0)))))

(fn GetItemCountForAgent
  (params (Controller :type Controller%20Object%20Reference) (Index :type Integer))
  (bind _lock_array (Variables|Default|GetLock%20Array))
  (Utilities|IsValid Controller
    (:Is%20Valid
      (if (Utilities|Array|IsValidIndex _lock_array Index)
        (bind _output_get (LocalVariables|SetLockToCheck (Utilities|Array|Get(acopy) _lock_array Index)))
        (Utilities|IsValid _output_get
          (:Is%20Valid
            (bind _returnvalue (Class|FortCreativeKeyLockComponent|BlueprintGetKeyAmountonPlayer _output_get (Pawn|GetControlledPawn Controller)))
            (return _returnvalue))
          (:Is%20Not%20Valid
            _))
        (else
          _))
      (return -1))
    (:Is%20Not%20Valid
      _)))

(fn GetKeyItemConsumedScore
  (params (KeyItemIndex :type Integer))
  (return (select (>= (Utilities|Select :Index KeyItemIndex :Option%200 (Variables|Verse|GetScoreOnKeyItem1%20Consumed%20Override) :Option%201 (Variables|Verse|GetScoreOnKeyItem2%20Consumed%20Override) :Option%202 (Variables|Verse|GetScoreOnKeyItem3%20Consumed%20Override)) 0) (Utilities|Select :Index KeyItemIndex :Option%200 (Variables|Verse|GetScoreOnKeyItem1%20Consumed%20Override) :Option%201 (Variables|Verse|GetScoreOnKeyItem2%20Consumed%20Override) :Option%202 (Variables|Verse|GetScoreOnKeyItem3%20Consumed%20Override)) (Utilities|Select :Index KeyItemIndex :Option%200 (Variables|UserOptions|GetScoreOnKeyItem1%20Consumed) :Option%201 (Variables|UserOptions|GetScoreOnKeyItem2%20Consumed) :Option%202 (Variables|UserOptions|GetScoreOnKeyItem3%20Consumed)))))

(fn GetKeyItemCountRequired
  (params (KeyItemIndex :type Integer))
  (switch int KeyItemIndex
    (:Default
      (return 0))
    (:0
      (return (Variables|/Script/FortniteGame.FortCreativeKeyLockComponent|GetRequiredKeysToUnlockAmount (Variables|Default|GetBP_KeyLockComponent))))
    (:1
      (return (Variables|/Script/FortniteGame.FortCreativeKeyLockComponent|GetRequiredKeysToUnlockAmount (Variables|Default|GetBP_KeyLockComponent2))))
    (:2
      (return (Variables|/Script/FortniteGame.FortCreativeKeyLockComponent|GetRequiredKeysToUnlockAmount (Variables|Default|GetBP_KeyLockComponent3))))))

(fn GetMeshComponents
  (return (Utilities|Array|MakeArray (Variables|Debug|GetStaticMeshComponent))))

(fn GetMusicSequencerHitSound
  (params (InstigatedBy :type Controller%20Object%20Reference))
  (return 0 0 0 0))

(fn GetMusicSequencerHitSoundLocation
  (params (InstigatedBy :type Controller%20Object%20Reference))
  (return "0, 0, 0"))

(fn GetPrimaryIcon
  (return (Utilities|ResolveSoftReference (Creative|UI|IconPicker|GetIconLibraryTexture (Variables|UserOptions|Basic|GetDisplayMainIcon)))))

(fn GetRemainingKeyItemCountRequired
  (params (KeyItemIndex :type Integer))
  (switch int KeyItemIndex
    (:Default
      (return 0))
    (:0
      (return (Variables|/Script/FortniteGame.FortCreativeKeyLockComponent|GetRemainingKeysToUnlockAmount (Variables|Default|GetBP_KeyLockComponent))))
    (:1
      (return (Variables|/Script/FortniteGame.FortCreativeKeyLockComponent|GetRemainingKeysToUnlockAmount (Variables|Default|GetBP_KeyLockComponent2))))
    (:2
      (return (Variables|/Script/FortniteGame.FortCreativeKeyLockComponent|GetRemainingKeysToUnlockAmount (Variables|Default|GetBP_KeyLockComponent3))))))

(fn GetSecondaryIcon
  (if (Variables|UserOptions|Basic|GetUseAltDisplayIcon)
    (return (Utilities|ResolveSoftReference (Creative|UI|IconPicker|GetIconLibraryTexture (Variables|UserOptions|Basic|GetAltDisplayIcon))))
    (else
      (return (Utilities|Map|Find (Variables|Visuals|GetOppositeIconByAction) (Utilities|ResolveSoftReference (Creative|UI|IconPicker|GetIconLibraryTexture (Variables|UserOptions|Basic|GetDisplayMainIcon))))))))

(fn GetSpawnedPickups
  (return 0))

(fn GetTeamColor
  (bind _activatingteam (Variables|UserOptions|GetActivatingTeam))
  (bind _returnvalue (Creative|TeamOption|ConverttoTeam _activatingteam))
  (bind _returnvalue_2 (Creative|TeamOption|IsTeamType _activatingteam))
  (bind _fortminigameprogress (Variables|Default|GetFortMinigameProgress))
  (bind _currentminigame (Variables|/Script/FortniteGame.FortMinigameLogicComponent|GetCurrentMinigame _fortminigameprogress))
  (bind _returnvalue_3 (Minigame|GetMinigameTeamComponent _currentminigame))
  (bind _returnvalue_4 (Utilities|Enum|LiteralenumEFortTeam "Spectator"))
  (bind _returnvalue_5 (select _returnvalue_2 _returnvalue _returnvalue_4))
  (Utilities|IsValid _currentminigame
    (:Is%20Valid
      (Utilities|IsValid _returnvalue_3
        (:Is%20Valid
          (return (Minigame|GameRules|GetTeamColorIndex _returnvalue_3 _returnvalue_5)))
        (:Is%20Not%20Valid
          _)))
    (:Is%20Not%20Valid
      _))
  (return (Math|Conversions|ToInteger(Byte) (- _returnvalue_5 2))))

(fn HandleLocalPawnInventoryChanged
  (if (and (Variables|Gameplay|GetbIsEnabled) (Variables|Gameplay|GetbIsMinigameInProgress))
    (bind _cachedlocalcontroller (Variables|FortCreativeLockDevice|GetCachedLocalController)
      (:then
        (bind _output_get (LocalVariables|SetlocalPawn (Pawn|GetControlledPawn _cachedlocalcontroller)))
        (Utilities|IsValid _output_get
          (:Is%20Valid
            (if (TestingRequirements|AnyLocked)
              (Visuals|UpdatePreviewColors false (select (TestingRequirements|TestRequirements (Pawn|GetControlledPawn _cachedlocalcontroller)) "NewEnumerator1" "NewEnumerator0"))))))
      (:else))
    (else
      (Visuals|UpdatePreviewColors false (select (not (Variables|Gameplay|GetbIsMinigameInProgress)) "NewEnumerator1" "NewEnumerator0")))))

(fn HasItem
  (params (Item :type Item%20Variant%20Handle%20Structure))
  (for _index (range (Variables|UserOptions|GetNumberOfKeyItemSlots))
    (bind _result (ArrayMacros|GetValidIndex (Variables|Default|GetTrackedItems) _index)
      (:Valid%20Index
        (if (Creative|ItemVariant|Equal(ItemVariantHandle) _result (Variables|GetItem))
          (return true)))))
  (return false))

(fn IgnoreWorldGridScaleOverride
  (return false))

(fn IncrementNextItemtoAdd
  (Variables|Default|SetNextItemToAdd (Math|Integer|%25%28Integer%29 (+ (Variables|Default|GetNextItemToAdd) 1) (Variables|UserOptions|GetNumberOfKeyItemSlots)))
  (Networking|ForceNetUpdate))

(fn InitializeVisibility
  (bind _returnvalue (Utilities|Select :Index (Variables|UserOptions|GetVisibleDuringGame) :No "NewEnumerator1" :Yes "NewEnumerator3" :HologramOnly "NewEnumerator3"))
  (bind _visibleduringgame (Variables|UserOptions|GetVisibleDuringGame))
  ;[0 1 0.3626 0.25] Initialize Visibility
  (Initialization|InitializeWithSceneComponents (Variables|Default|GetVisibleInGameComponent_StaticMesh) (Utilities|Array|MakeArray (Variables|Debug|GetStaticMeshComponent)) (Utilities|Map|MakeMap) (Utilities|Select :Index _visibleduringgame :No "NewEnumerator1" :Yes "NewEnumerator3" :HologramOnly "NewEnumerator0"))
  (Initialization|InitializeWithSceneComponents (Variables|Default|GetVisibleInGameComponent_Hologram1) (Utilities|Array|MakeArray (Variables|Default|GetLockHologram_11) (Variables|Default|GetHologram_Cone) (Variables|Default|GetHologram_Grid)) (Utilities|Map|MakeMap) _returnvalue)
  (Initialization|InitializeWithSceneComponents (Variables|Default|GetVisibleInGameComponent_Hologram2) (Utilities|Array|MakeArray (Variables|Default|GetLockHologram_22)) (Utilities|Map|MakeMap) (select (>= (Variables|UserOptions|GetNumberOfKeyItemSlots) 2) _returnvalue "NewEnumerator0"))
  (Initialization|InitializeWithSceneComponents (Variables|Default|GetVisibleInGameComponent_Hologram3) (Utilities|Array|MakeArray (Variables|Default|GetLockHologram_33)) (Utilities|Map|MakeMap) (select (>= (Variables|UserOptions|GetNumberOfKeyItemSlots) 3) _returnvalue "NewEnumerator0")))

(fn IsEnabled
  (return false))

(fn IsHoldingItemForPawn
  (params (Item :type Item%20Variant%20Handle%20Structure) (Pawn :type Pawn%20Object%20Reference))
  (Utilities|IsValid Pawn
    (:Is%20Valid
      (if (and (Variables|UserOptions|GetShowKeycardDirection) (Variables|Gameplay|GetbIsEnabled))
        (if (Creative|ItemVariant|IsValid(ItemVariantHandle) Item)
          (return (and (CallFunction|HasItem (Variables|GetItem)) (TestingRequirements|Team&ClassMatch (Variables|GetPawn))))
          (else
            (return (TestingRequirements|TestRequirements Pawn))))
        (else
          _)))
    (:Is%20Not%20Valid
      _))
  (return false))

(fn IsTargetableByLinearMover
  (return true))

(fn LocalPlayerUpdated
  (bind _returnvalue (Fortnite|GetFirstLocalFortPlayerController true))
  (bind _returnvalue_2 (Game|Player|GetPlayerPawn _returnvalue))
  (Networking|CanExecuteCosmeticEvents
    (:True
      (Visuals|UpdateColors)
      (Visuals|UpdatePreviewColors true "NewEnumerator1")
      (Utilities|IsValid _returnvalue_2
        (:Is%20Valid
          (Visuals|UpdateDeviceDisplayIcon (Variables|Gameplay|GetIconsWereToggled))
          (CallFunction|HandleLocalPawnInventoryChanged)
          (CallFunction|PawnOverlapUpdated (and (Collision|IsOverlappingActor (Variables|Default|GetShowPreviewArea) _returnvalue_2 :!class /Script/Engine.PrimitiveComponent) (TestingRequirements|CanLocalTeam&ClassInteract)) _returnvalue_2))))))

(fn LocalOnFailedInteract
  (params (InteractingPawn :type Fort%20Player%20Pawn%20Object%20Reference))
  (|CallParentFunction InteractingPawn)
  (FailedAction|UnabletoPerformAction InteractingPawn true (GameplayTags|MakeLiteralGameplayTagContainer "(GameplayTags=((TagName=\"Activate.Failed.Tags.Missing\")))")))

(fn MigrateLegacyUserOptions
  (|CallParentFunction)
  (Utilities|FlowControl|Sequence
    (:then_0
      (Utilities|FlowControl|Sequence
        (:then_0
          (if (< (Variables|UserOptions|Basic|GetInteractTime) 0)
            (Variables|UserOptions|Basic|SetAllowInteraction false)
            (Variables|UserOptions|Basic|SetInteractTime (Class|GetClassDefaults "/CreativeCoreDevices/Device_ConditionalButton_V2.Device_ConditionalButton_V2_C"))
            (LocalVariables|SetbMigratedProperties true)))
        (:then_1
          (if (!= (Variables|LegacyUserOptions|GetDisplayIconIndexLegacy) (Class|GetClassDefaults "/CreativeCoreDevices/Device_ConditionalButton_V2.Device_ConditionalButton_V2_C"))
            (Variables|UserOptions|Basic|SetDisplayMainIcon (Utilities|Array|Get(acopy) (Variables|Visuals|GetIconStructs) (Variables|LegacyUserOptions|GetDisplayIconIndexLegacy)))
            (Variables|LegacyUserOptions|SetDisplayIconIndexLegacy (Class|GetClassDefaults "/CreativeCoreDevices/Device_ConditionalButton_V2.Device_ConditionalButton_V2_C"))
            (LocalVariables|SetbMigratedProperties true)))
        (:then_2
          (if (!= (Variables|LegacyUserOptions|GetDisplayAltIconIndexLegacy) (Class|GetClassDefaults "/CreativeCoreDevices/Device_ConditionalButton_V2.Device_ConditionalButton_V2_C"))
            (Variables|UserOptions|Basic|SetAltDisplayIcon (Utilities|Array|Get(acopy) (Variables|Visuals|GetIconStructs) (Variables|LegacyUserOptions|GetDisplayAltIconIndexLegacy)))
            (Variables|LegacyUserOptions|SetDisplayAltIconIndexLegacy (Class|GetClassDefaults "/CreativeCoreDevices/Device_ConditionalButton_V2.Device_ConditionalButton_V2_C"))
            (Variables|UserOptions|Basic|SetUseAltDisplayIcon true)
            (LocalVariables|SetbMigratedProperties true)))
        (:then_3
          (if (!= (Variables|LegacyUserOptions|GetUseHologramColorLegacy) (Class|GetClassDefaults "/CreativeCoreDevices/Device_ConditionalButton_V2.Device_ConditionalButton_V2_C"))
            (Variables|UserOptions|SetUseColorForHologram true)
            (Variables|LegacyUserOptions|SetUseHologramColorLegacy (Class|GetClassDefaults "/CreativeCoreDevices/Device_ConditionalButton_V2.Device_ConditionalButton_V2_C"))
            (LocalVariables|SetbMigratedProperties true)))
        (:then_4
          ;[1 1 1] There is something overwritting ToggleIconOnUse to false no matter what it is set to, so I renamed it and converted manually.
          (if (!= (Variables|LegacyUserOptions|GetToggleIconOnUseLegacy) (Class|GetClassDefaults "/CreativeCoreDevices/Device_ConditionalButton_V2.Device_ConditionalButton_V2_C"))
            (Variables|UserOptions|SetToggleIconOnUseNew true)
            (Variables|LegacyUserOptions|SetToggleIconOnUseLegacy (Class|GetClassDefaults "/CreativeCoreDevices/Device_ConditionalButton_V2.Device_ConditionalButton_V2_C"))
            (LocalVariables|SetbMigratedProperties true)))
        (:then_5
          (if (== (Creative|TeamOption|GetLegacyTeamIndex (Variables|UserOptions|GetActivatingTeam)) (LocalVariables|GetInvalidTeamIndex))
            (Variables|UserOptions|SetActivatingTeam (Class|GetClassDefaults "/CreativeCoreDevices/Device_ConditionalButton_V2.Device_ConditionalButton_V2_C"))
            (LocalVariables|SetbMigratedProperties true)))
        (:then_6
          (if (< (Variables|UserOptions|GetInteractionRadius) 0)
            (Variables|UserOptions|SetInteractionRadius 0.0)
            (LocalVariables|SetbMigratedProperties true)))))
    (:then_1
      (if (LocalVariables|GetbMigratedProperties)
        (OptionData|RefreshPropertyOverrides (Variables|Default|GetToyOptionsComponent))))))

(fn OnRep_bDeviceIsResetting
  (Networking|CanExecuteCosmeticEvents
    (:True
      (if (Variables|Gameplay|GetbDeviceIsResetting)
        (CallFunction|OnDeviceReset)))))

(fn OnRep_bIsEnabled
  (CallFunction|UpdatePreviewVisibility)
  (Class|FortCreativeInteractVisualizationComponent|SetRadius (Variables|Default|GetInteractVisualizationComponent) (Math|Float|SelectFloat (Variables|UserOptions|GetInteractionRadius) 0.0 (Variables|Gameplay|GetbIsEnabled))))

(fn OnRep_CurrentItemCaptured
  (Networking|CanExecuteCosmeticEvents
    (:True
      (CallFunction|OnItemCaptured (Variables|CurrentPickupData|GetCurrentItemCaptured)))))

(fn OnRep_IconsWereToggled
  (Networking|CanExecuteCosmeticEvents
    (:True
      (Visuals|UpdateDeviceDisplayIcon (Variables|Gameplay|GetIconsWereToggled)))))

(fn OnRep_NextItemToAdd
  (CallFunction|UpdateNextItemToAdd))

(fn OnWorldInventoryChanged
  (params (bIsAllowedToFly :type Boolean))
  (CallFunction|HandleLocalPawnInventoryChanged))

(fn PawnHasKeys
  (params (Pawn :type Pawn%20Object%20Reference))
  (bind _lock_array (Variables|Default|GetLock%20Array))
  (Utilities|IsValid Pawn
    (:Is%20Valid
      (if (or (not (TestingRequirements|AnyLocked)) (not (Variables|Gameplay|GetbIsMinigameInProgress)))
        (return true)
        (else
          (for _index (range (Variables|UserOptions|GetNumberOfKeyItemSlots))
            (if (Class|FortCreativeKeyLockComponent|IsLocked (Utilities|Array|Get(acopy) _lock_array _index))
              (bind _returnvalue (Class|FortCreativeKeyLockComponent|BlueprintHasKeystoInteract (Utilities|Array|Get(acopy) _lock_array _index) Pawn))
              (if _returnvalue
                (Math|Integer|IncrementInt (LocalVariables|GetHasKeys)))))
          (if (Variables|UserOptions|Basic|GetAllKeyItemsRequiredAtOnce)
            (return (== (LocalVariables|GetHasKeys) (Variables|UserOptions|GetNumberOfKeyItemSlots)))
            (else
              (return (> (LocalVariables|GetHasKeys) 0)))))))
    (:Is%20Not%20Valid
      (return false))))

(fn PawnIsHoldingItem
  (params (Pawn :type Pawn%20Object%20Reference) (KeyIndex :type Integer :default -1) (ForceCheck :type Boolean))
  (if (or (Math|Boolean|ANDBoolean (TestingRequirements|AnyLocked) (Variables|Gameplay|GetbIsMinigameInProgress) (Variables|UserOptions|GetbRequireHoldingItem)) ForceCheck)
    ;[0 0.6864 0.7396 0.292] Get the current equipped item variant (Only method to get current equipped item on the server is to get the currently equipped weapon (turns out all equippable items are weapons of some form as far as I can tell. Confirmed with BR design)
    (bind _asfort_player_pawn_athena (Utilities|Casting|CastToFortPlayerPawnAthena Pawn))
    (Utilities|IsValid (Weapon|GetCurrentWeapon _asfort_player_pawn_athena)
      (:Is%20Valid
        (Utilities|IsValid (FortWeapon|GetInventoryItem (Weapon|GetCurrentWeapon _asfort_player_pawn_athena))
          (:Is%20Valid
            (LocalVariables|SetLocalItemVariant (Item|GetItemVariantHandle (FortWeapon|GetInventoryItem (Weapon|GetCurrentWeapon _asfort_player_pawn_athena))))
            ;[0 0.8019 1 0.333] If we have a key index, check it directly.
            (bind _result (ArrayMacros|GetValidIndex (Variables|Default|GetTrackedItems) (Variables|GetKeyIndex))
              (:Valid%20Index
                (return (== _result (LocalVariables|GetLocalItemVariant))))
              (:Invalid%20Index
                ;[0 0.6864 0.7396 0.292] Else iterate to check all possible keys.
                (for _array_element (Variables|Default|GetTrackedItems)
                  (if (== (LocalVariables|GetLocalItemVariant) _array_element)
                    (return true)))
                (return false))))
          (:Is%20Not%20Valid
            _)))
      (:Is%20Not%20Valid
        _))
    (return false)
    (else
      (return true))))

(fn PlayerChangedClass
  (params (PlayerState :type Fort%20Player%20State%20Object%20Reference) (ClassSlotIndex :type Byte))
  (Networking|CanExecuteCosmeticEvents
    (:True
      (if (Utilities|Equal(Object) (Fortnite|GetFortPlayerControllerfromActor PlayerState) (Fortnite|GetFirstLocalFortPlayerController true))
        (TestingRequirements|LocalPlayerUpdated)))))

(fn RefreshNextItemtoAdd
  ;[1 1 1] Refreshes the next item to add (in case of change in number of total items)
  (Networking|SwitchHasAuthority
    (:Authority
      (Variables|Default|SetNextItemToAdd (Math|Integer|%25%28Integer%29 (Variables|Default|GetNextItemToAdd) (Variables|UserOptions|GetNumberOfKeyItemSlots))))))

(fn SendScoreStatEvent
  (params (Player%20Controller :type Fort%20Player%20Controller%20Object%20Reference) (KeyIndex :type Integer))
  (bind _keyindex (Variables|GetKeyIndex))
  (bind _scoretoadd2 (Variables|Gameplay|GetScoreToAdd2))
  (bind _scoretoadd1 (Variables|Gameplay|GetScoreToAdd1))
  (bind _scoretoadd (Variables|Gameplay|GetScoreToAdd))
  (bind _returnvalue (Utilities|Select :Index _keyindex :Option%200 _scoretoadd :Option%201 _scoretoadd1 :Option%202 _scoretoadd2))
  (Utilities|IsValid Player%20Controller
    (:Is%20Valid
      (bind _score (Verse|GetKeyItemConsumedScore (Variables|GetKeyIndex)))
      (LocalVariables|SetScoreAtIndex (* _score _returnvalue))
      (if (> (* _score _returnvalue) 0)
        (LocalVariables|SetTargetTags (GameplayTags|GetOwnedGameplayTags self :!class /Script/GameplayTags.GameplayTagAssetInterface))
        (Utilities|FlowControl|Sequence
          (:then_0
            (if (GameplayTags|IsGameplayTagValid (Variables|Default|GetUnlockEventTag))
              (GameplayTags|AddGameplayTag (LocalVariables|GetTargetTags) (Variables|Default|GetUnlockEventTag) :!class /Script/GameplayTags.BlueprintGameplayTagLibrary)))
          (:then_1
            (bind _returnvalue_2 (CreativeDevices|GetMinigameStatsComp (Variables|GetPlayer%20Controller) "Athena"))
            (LocalVariables|Setlocal%20MinigameStatsComp _returnvalue_2)
            (bind _local_minigamestatscomp (LocalVariables|Getlocal%20MinigameStatsComp)
              (:then
                (Class|FortMinigameStatsComponent|SendStatEventwithTags _local_minigamestatscomp (Variables|GetPlayer%20Controller) "MinigameDynamicEvent" self (LocalVariables|GetTargetTags) (Class|FortMinigameStatsComponent|GetSourceandContextTags _local_minigamestatscomp (Variables|GetPlayer%20Controller)) (.OutContextTags (Class|FortMinigameStatsComponent|GetSourceandContextTags _local_minigamestatscomp (Variables|GetPlayer%20Controller))) (LocalVariables|GetScoreAtIndex)))
              (:else))))))))

(fn SetKeysRequired
  (bind _lock_array (Variables|Default|GetLock%20Array))
  ;[1 1 1] Setting Key Requirements
  (Class|FortCreativeKeyLockComponent|SetRequiredKeys (Variables|Default|GetBP_KeyLockComponent) (Math|Integer|SelectInt (Variables|UserOptions|Basic|GetKeyItemsRequired) (Variables|Default|GetStackSize_1) (Variables|UserOptions|GetKeyItemsRequired_Override)))
  (Class|FortCreativeKeyLockComponent|SetRequiredKeys (Variables|Default|GetBP_KeyLockComponent2) (Math|Integer|SelectInt (Variables|UserOptions|GetSecondKeyItemsRequired) (Variables|Default|GetStackSize_2) (Variables|UserOptions|GetSecondKeyItemsRequired_Override)))
  (Class|FortCreativeKeyLockComponent|SetRequiredKeys (Variables|Default|GetBP_KeyLockComponent3) (Math|Integer|SelectInt (Variables|UserOptions|GetThirdKeyItemsRequired) (Variables|Default|GetStackSize_3) (Variables|UserOptions|GetThirdKeyItemsRequired_Override)))
  (Class|FortCreativeKeyLockComponent|SetRequiredAllKeysatOne _lock_array (Variables|UserOptions|Basic|GetAllKeyItemsRequiredAtOnce))
  (Class|FortCreativeKeyLockComponent|SetShouldConsumeKey _lock_array (Variables|UserOptions|Basic|GetConsumeKeyItems)))

(fn SetTriggerCollision
  (params (On :type Boolean))
  (return false))

(fn SetDisplayIcon
  (params (bIsPrimaryIcon :type Boolean) (IconTexture :type Texture%20Object%20Reference))
  (Rendering|Material|SetTextureParameterValue (Variables|Default|GetCachedDisplayMID) (Utilities|String|StringToName (Utilities|String|SelectString "State A Icon" "State B Icon" bIsPrimaryIcon)) IconTexture)
  (SetScreenElement|SetIcon (Variables|Default|GetHologram%20Array) bIsPrimaryIcon IconTexture))

(fn SetEnableState
  (params (Enabled :type Boolean :default true))
  (Networking|SwitchHasAuthority
    (:Authority
      (Variables|Gameplay|SetbIsEnabled Enabled)
      (Networking|FlushNetDormancy))))

(fn SetHologramVisibility
  (params (NumHolograms :type Integer))
  ;[1 1 1] Scale of Visuals
  (Transformation|SetRelativeScale3D (Variables|Default|GetHologram_Cone) (Math|Vector|MakeVector (Math|Float|Max(Float) 0.01 (Math|Conversions|ToFloat(Integer) (Variables|GetNumHolograms))) 0.85 1.0))
  ;[1 1 1] Adjust location so the locks are always centered.
  (Transformation|SetRelativeLocation (Variables|Default|GetLockHologram_11) (Math|Vector|MakeVector (Utilities|Select :Index (Variables|GetNumHolograms) :Option%200 0.0 :Option%201 0.0 :Option%202 -3e+01 :Option%203 -65.0)) false false))

(fn SetInteractionText
  (params (InteractionText :type Localizable%20Message%20Structure))
  (Networking|SwitchHasAuthority
    (:Authority
      (Class|FortLocalizableMessageComponent|LocalizeandSetMessage (Variables|Default|GetFortLocalizableMessage) (Variables|GetInteractionText) "InteractionStringOverride" "InteractText"))))

(fn SetInteractionTime
  (params (InteractionTime :type Float%20%28double-precision%29))
  (Networking|SwitchHasAuthority
    (:Authority
      (Variables|Verse|SetInteractTimeOverride (Math|Float|Clamp(Float) InteractionTime 0.0 2e+01)))))

(fn SetKeyItemConsumedScore
  (params (KeyItemIndex :type Integer) (Score :type Integer))
  (bind _returnvalue (Math|Integer|Clamp(Integer) Score 0 50))
  (Networking|SwitchHasAuthority
    (:Authority
      (switch int KeyItemIndex
        (:0
          (Variables|Verse|SetScoreOnKeyItem1%20Consumed%20Override _returnvalue))
        (:1
          (Variables|Verse|SetScoreOnKeyItem2%20Consumed%20Override _returnvalue))
        (:2
          (Variables|Verse|SetScoreOnKeyItem3%20Consumed%20Override _returnvalue))))))

(fn SetKeyItemCountRequired
  (params (KeyItemIndex :type Integer) (Count :type Integer))
  (bind _returnvalue (Math|Integer|Clamp(Integer) Count 0 10000000))
  (Networking|SwitchHasAuthority
    (:Authority
      (switch int KeyItemIndex
        (:0
          (Class|FortCreativeKeyLockComponent|SetRequiredKeys (Variables|Default|GetBP_KeyLockComponent) _returnvalue))
        (:1
          (Class|FortCreativeKeyLockComponent|SetRequiredKeys (Variables|Default|GetBP_KeyLockComponent2) _returnvalue))
        (:2
          (Class|FortCreativeKeyLockComponent|SetRequiredKeys (Variables|Default|GetBP_KeyLockComponent3) _returnvalue))))))

(fn SetupLockArrays
  (Variables|Default|SetLock%20Array (Utilities|Array|MakeArray (Variables|Default|GetBP_KeyLockComponent) (Variables|Default|GetBP_KeyLockComponent2) (Variables|Default|GetBP_KeyLockComponent3)))
  (bind _output_get (Variables|Default|SetHologram%20Array (Utilities|Array|MakeArray (Variables|Default|GetLockHologram_11) (Variables|Default|GetLockHologram_22) (Variables|Default|GetLockHologram_33))))
  (for (_array_element _array_index) _output_get
    (Class|BPcLockHologramV2|BindtoEvents _array_element (Utilities|Array|Get(acopy) (Variables|Default|GetLock%20Array) _array_index))))

(fn SetupPreviewPreAnimation
  (params (InstigatorCanInteract :type Boolean))
  (bind _fortminigameprogress (Variables|Default|GetFortMinigameProgress))
  (bind _currentminigame (Variables|/Script/FortniteGame.FortMinigameLogicComponent|GetCurrentMinigame _fortminigameprogress))
  (SetScreenElement|ShowKeysRequired (Variables|Default|GetHologram%20Array) true)
  (Utilities|IsValid _currentminigame
    (:Is%20Valid
      (Minigame|IsInProgress _currentminigame
        (:Yes
          _)))
    (:Is%20Not%20Valid
      (SetScreenElement|UpdateScreenKey (Variables|Default|GetHologram%20Array))))
  (Visuals|UpdatePreviewColors false (select InstigatorCanInteract "NewEnumerator1" "NewEnumerator0"))
  (Components|Timeline|Stop (Variables|Device_ConditionalButton_V2|GetHide%20Preview)))

(fn SetupTeam
  (Networking|SwitchHasAuthority
    (:Authority
      (CallFunction|SetTeam (Creative|TeamOption|ConverttoTeam (Variables|UserOptions|GetActivatingTeam))))))

(fn ShouldInvisiblePropsBeAllowed
  (return false))

(fn Team&ClassMatch
  (params (Actor :type Actor%20Object%20Reference))
  (return (or (and (TestingRequirements|CanTeamInteract Actor) (TestingRequirements|CanClassInteract Actor)) (not (Variables|Gameplay|GetbIsMinigameInProgress)))))

(fn TestRequirements
  (params (Pawn :type Pawn%20Object%20Reference))
  (return (or (Math|Boolean|ANDBoolean (TestingRequirements|PawnHasKeys Pawn) (TestingRequirements|Team&ClassMatch Pawn) (Variables|Gameplay|GetbIsEnabled)) (not (Variables|Gameplay|GetbIsMinigameInProgress)))))

(fn TickPreviewAnimation
  (params (AnimTime :type Float%20%28double-precision%29))
  (PlayAnimations|UpdateSpawnAnimation (Variables|Default|GetHologram%20Array) AnimTime)
  (Rendering|Material|SetScalarParameterValue (Material|GetMID (Variables|Default|GetHologram_Cone)) "Spawn Animation Time" AnimTime :!class /Script/Engine.MaterialInstanceDynamic)
  (Rendering|Material|SetScalarParameterValue (Material|GetMID (Variables|Default|GetHologram_Grid)) "Spawn Animation Time" AnimTime :!class /Script/Engine.MaterialInstanceDynamic))

(fn ToggleOnUseEnabled
  (return (Variables|UserOptions|GetToggleIconOnUseNew)))

(fn UnlockAll
  (params (Pawn :type Pawn%20Object%20Reference))
  (for _index (range (Variables|UserOptions|GetNumberOfKeyItemSlots))
    (bind _output_get (LocalVariables|SetCurrentKeyIndex _index))
    (if (not (Utilities|Array|ContainsItem (Variables|Default|GetCompletedKeyIndexes) _output_get))
      (bind _returnvalue (Class|FortCreativeKeyLockComponent|Unlock (Utilities|Array|Get(acopy) (Variables|Default|GetLock%20Array) (LocalVariables|GetCurrentKeyIndex)) (Variables|GetPawn)))
      (if (and (Variables|UserOptions|GetbAddConsumedItemsToScore) (Variables|UserOptions|Basic|GetConsumeKeyItems))
        (CallFunction|SendScoreStatEvent (Fortnite|GetFortPlayerControllerfromActor (Variables|GetPawn)) (LocalVariables|GetCurrentKeyIndex))
        (else
          _))
      (if (not _returnvalue)
        (LocalVariables|SetNotUnlocked true)
        (else
          (Utilities|Array|Add (Variables|Default|GetCompletedKeyIndexes) (LocalVariables|GetCurrentKeyIndex))))))
  (Variables|Gameplay|SetScoreToAdd 0)
  (Variables|Gameplay|SetScoreToAdd1 0)
  (Variables|Gameplay|SetScoreToAdd2 0)
  (return (not (LocalVariables|GetNotUnlocked))))

(fn UpdateColors
  (bind _cachedbodymid (Variables|Default|GetCachedBodyMID))
  (if (Utilities|Enum|Equal(Enum) (Variables|UserOptions|Basic|GetColorType) "NewEnumerator1")
    (LocalVariables|SetDesiredColor (Math|Conversions|ToInteger(Byte) (Variables|UserOptions|Basic|GetDirectColor)))
    (else
      (LocalVariables|SetDesiredColor (Math|Integer|Max(Integer) (Visuals|GetTeamColor)))))
  (Material|ReattachOrCreateMID (Variables|Debug|GetStaticMeshComponent) _cachedbodymid)
  (Rendering|Material|SetScalarParameterValue _cachedbodymid "TeamColor" (Math|Conversions|ToFloat(Integer) (LocalVariables|GetDesiredColor)) :!class /Script/Engine.MaterialInstanceDynamic)
  (Rendering|Material|SetMaterial (Variables|Debug|GetStaticMeshComponent) 0 (Variables|Default|GetCachedBodyMID)))

(fn UpdateHolograms
  (Networking|CanExecuteCosmeticEvents
    (:True
      (ComponentOrders|Hologram|SetHologramVisibility (Variables|UserOptions|GetNumberOfKeyItemSlots))
      (Class|BPcLockHologramV2|InitializeComponent (Variables|Default|GetHologram%20Array)))))

(fn UpdateKeysRequiredDisplayEditorOnly
  (Class|BPcLockHologramV2|UpdateKeysRemaining (Variables|Default|GetLockHologram_11) (Math|Integer|SelectInt (Variables|Default|GetStackSize_1) (Variables|UserOptions|Basic|GetKeyItemsRequired) (Variables|UserOptions|GetKeyItemsRequired_Override)))
  (Class|BPcLockHologramV2|UpdateKeysRemaining (Variables|Default|GetLockHologram_22) (Math|Integer|SelectInt (Variables|Default|GetStackSize_2) (Variables|UserOptions|GetSecondKeyItemsRequired) (Variables|UserOptions|GetSecondKeyItemsRequired_Override)))
  (Class|BPcLockHologramV2|UpdateKeysRemaining (Variables|Default|GetLockHologram_33) (Math|Integer|SelectInt (Variables|Default|GetStackSize_3) (Variables|UserOptions|GetThirdKeyItemsRequired) (Variables|UserOptions|GetThirdKeyItemsRequired_Override))))

(fn UpdateNextItemtoAdd
  (if (not (Variables|Gameplay|GetbIsMinigameInProgress))
    (Networking|SwitchHasAuthority
      (:Remote
        (for (_array_element _array_index) (Variables|Default|GetHologram%20Array)
          (Class|BPcLockHologramV2|UpdateScreenEdgeForSelection _array_element (Math|Boolean|ANDBoolean (== _array_index (Variables|Default|GetNextItemToAdd)) (not (Class|CreativeCommonDeviceFunctionLibrary|IsLogicinPlayMode (Variables|Default|GetFortMinigameProgress))) (> (Variables|UserOptions|GetNumberOfKeyItemSlots) 1))))))))

(fn UpdateButtonCollision
  (bind _staticmeshcomponent (Variables|Debug|GetStaticMeshComponent))
  (bind _fortminigameprogress (Variables|Default|GetFortMinigameProgress))
  (bind _currentminigame (Variables|/Script/FortniteGame.FortMinigameLogicComponent|GetCurrentMinigame _fortminigameprogress))
  (Utilities|IsValid _currentminigame
    (:Is%20Valid
      (if (Minigame|IsinProgress _currentminigame)
        (switch Utilities|FlowControl|Switch|SwitchonEFortLockDeviceVisibilityDuringGames (Variables|UserOptions|GetVisibleDuringGame)
          (:No
            (Collision|SetCollisionResponsetoAllChannels (Variables|Debug|GetStaticMeshComponent) "ECR_Ignore") ; Ignore Everything
            (Collision|SetCollisionResponsetoChannel (Variables|Debug|GetStaticMeshComponent) "ECC_GameTraceChannel1" "ECR_Block") ; All collision disabled apart from Interaction
          )
          (:Yes
            (Collision|SetCollisionProfileName (Variables|Debug|GetStaticMeshComponent) "FortBuildingMeshPhysics" true :!class /Script/Engine.PrimitiveComponent) ; Back to Defaults
          )
          (:HologramOnly
            _))
        (if (Variables|UserOptions|GetActivatedBySequencers)
          (Networking|SwitchHasAuthority
            (:Authority
              (Collision|SetCollisionResponsetoChannel _staticmeshcomponent "ECC_WorldDynamic" "ECR_Overlap") ; If we are triggered by sequencers, enable collision with world dynamic (the collision channel\nfor the MovingTrigger and allow Overlap events to occur)
              (Variables|/Script/Engine.PrimitiveComponent|SetbGenerateOverlapEvents true _staticmeshcomponent))))
        (else
          _)))
    (:Is%20Not%20Valid
      _)))

(fn UpdateDeviceDisplayIcon
  (params (SwitchConfiguredIcons :type Boolean))
  (Material|ReattachOrCreateMID (Variables|Debug|GetStaticMeshComponent) (Variables|Default|GetCachedDisplayMID) 1)
  (Utilities|FlowControl|Sequence
    (:then_0
      (Visuals|SetDisplayIcon true (select (TestingRequirements|CanLocalTeam&ClassInteract) (select (Variables|GetSwitchConfiguredIcons) (Visuals|GetSecondaryIcon) (Visuals|GetPrimaryIcon)) "/Game/Creative/Devices/Lock_and_Key/Textures/T_LockKeyRequirements_M.T_LockKeyRequirements_M")))
    (:then_1
      (Visuals|SetDisplayIcon false (select (TestingRequirements|CanLocalTeam&ClassInteract) (select (Variables|GetSwitchConfiguredIcons) (Visuals|GetPrimaryIcon) (Visuals|GetSecondaryIcon)) "/Game/Creative/Devices/Lock_and_Key/Textures/T_LockKeyRequirements_M.T_LockKeyRequirements_M")))))

(fn UpdateDeviceDisplayState
  (params (NewState :type EConditionalButtonDisplayState%20Enum :default "NewEnumerator0"))
  (bind _staticmeshcomponent (Variables|Debug|GetStaticMeshComponent))
  (switch Utilities|FlowControl|Switch|SwitchonEConditionalButtonDisplayState NewState
    (:NewEnumerator0
      (Rendering|Material|SetScalarParameterValue (Material|GetMIDfromMaterial _staticmeshcomponent 1) "Locked - Unlocked" 1.0 :!class /Script/Engine.MaterialInstanceDynamic))
    (:NewEnumerator1
      (Rendering|Material|SetScalarParameterValue (Material|GetMID _staticmeshcomponent) "Locked - Unlocked" 0.0 :!class /Script/Engine.MaterialInstanceDynamic))))

(fn UpdatePreviewColors
  (params (AlwaysUpdate :type Boolean) (PreviewColor :type ELockDevicePreviewColor%20Enum :default "NewEnumerator0"))
  (bind _cacheddisplaymid (Variables|Default|GetCachedDisplayMID))
  (if (or AlwaysUpdate (Utilities|Enum|NotEqual(Enum) PreviewColor (Variables|Visuals|GetCachedPreviewColor)))
    (Variables|Visuals|SetCachedPreviewColor PreviewColor)
    (Utilities|NotforPreviewing
      (:True
        (Networking|CanExecuteCosmeticEvents
          (:True
            (bind (_icon _edge _background) (CallFunction|ValidToColor (Variables|GetPreviewColor)))
            (Rendering|Material|SetVectorParameterValue (Material|GetMID (Variables|Default|GetHologram_Cone)) "Cone Color" _edge :!class /Script/Engine.MaterialInstanceDynamic)
            (Rendering|Material|SetVectorParameterValue (Material|GetMID (Variables|Default|GetHologram_Grid)) "Grid Color" (Math|Color|MakeColor (Math|Color|BreakColor _edge) (.G (Math|Color|BreakColor _edge)) (.B (Math|Color|BreakColor _edge))) :!class /Script/Engine.MaterialInstanceDynamic)
            (Material|ReattachOrCreateMID (Variables|Debug|GetStaticMeshComponent) (Variables|Default|GetCachedDisplayMID) 1)
            (Rendering|Material|SetVectorParameterValue _cacheddisplaymid "Icon Color" _icon :!class /Script/Engine.MaterialInstanceDynamic)
            (Rendering|Material|SetVectorParameterValue _cacheddisplaymid "Background Color" _background :!class /Script/Engine.MaterialInstanceDynamic)
            (Class|BPcLockHologramV2|UpdateScreenColorfromLocalPawn (Variables|Default|GetHologram%20Array))))))))

(fn UpdatePreviewVisibility
  (bind (_in_range _can_interact _color) (CallFunction|GetLocalPlayerDetails))
  (Networking|CanExecuteCosmeticEvents
    (:True
      (Visuals|UpdatePreviewColors true _color)
      (if _in_range
        (CallFunction|ShowKeyPreview _can_interact)
        (CallFunction|HandleLocalPawnInventoryChanged)
        (else
          (CallFunction|HideKeyPreview))))))

(fn ConstructionScript
  (|CallParentFunction)
  (CallFunction|SetupLockArrays)
  (ComponentOrders|Hologram|SetHologramVisibility (Variables|UserOptions|GetNumberOfKeyItemSlots))
  ;[0.1377 1 0.1279] Run these to properly update hologram and button mesh in UEFN Editor.
  (Visuals|UpdateColors)
  (Visuals|UpdatePreviewColors false "NewEnumerator1"))

(fn ValidtoColor
  (params (IsValid :type ELockDevicePreviewColor%20Enum :default "NewEnumerator3"))
  (bind (_returnvalue_iconcolor_4_1d3fd4434324ff8fa1f9b29974146237 _returnvalue_edgecolor_5_22bce9674741b4f5054099a6fecc4593 _returnvalue_backgroundcolor_6_bd0b3efd426a06c9f1bf59bdc1f47dbd) (Utilities|Struct|BreakLockDeviceColor (select (or (Utilities|Enum|Equal(Enum) (Variables|GetIsValid) "NewEnumerator1") (not (Class|CreativeCommonDeviceFunctionLibrary|IsLogicinPlayMode (Variables|Default|GetFortMinigameProgress)))) (select (Variables|UserOptions|GetUseColorForHologram) (Utilities|Map|Find (Variables|VisualsColors|GetIcon%20Colors) (Utilities|Enum|BytetoEnumECreativeColor (Math|Conversions|ToByte(Integer) (Math|Integer|Max(Integer) (Utilities|Select :Index (Variables|UserOptions|Basic|GetColorType) :NewEnumerator0 0 :NewEnumerator1 (Math|Conversions|ToInteger(Byte) (Variables|UserOptions|Basic|GetDirectColor)) :NewEnumerator2 (Visuals|GetTeamColor) :NewEnumerator3 0))))) (Variables|VisualsColors|GetDefaultValidColor)) (Variables|VisualsColors|GetInvalidColor))))
  (return _returnvalue_iconcolor_4_1d3fd4434324ff8fa1f9b29974146237 _returnvalue_edgecolor_5_22bce9674741b4f5054099a6fecc4593 _returnvalue_backgroundcolor_6_bd0b3efd426a06c9f1bf59bdc1f47dbd))
