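; Fortnite Creative decal device that projects a user-configured decal with selectable material, mask, color, opacity and distance.
; Option load, option change and editor construction async-load material and mask, build a dynamic material instance, then update its parameters and decal scale on cosmetic-capable instances.
; CreativeColorComponent resolves team or custom color, which visual-update events write to ColorLight.
; MigrateLegacyUserOptions converts the old color knob and invalid None team, then refreshes property overrides.
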(event UserOptions|EventHandleUserOptionsLoaded
  (bind (_decal_mask_smallicon _decal_mask_largeicon) (Utilities|Struct|BreakCreativeIconOption (Variables|GetDecal%20Mask_SmallIcon)))
  (|CallParentFunction)
  (Networking|CanExecuteCosmeticEvents
    (:True
      ;[0.5 1 1] Getting the decal MID
      (bind _loadeddecalmid (Variables|Default|GetLoadedDecalMID)
        (:then
          (if (not (== _loadeddecalmid (Utilities|ResolveSoftReference (Variables|UserOptions|Basic|GetDecal%20Material))))
            (bind _loaded_object (Creative|Loading|AsyncLoadEditorSafe (Variables|UserOptions|Basic|GetDecal%20Material))
              (:Loaded
                (bind _asmaterial_interface (Utilities|Casting|CastToMaterialInterface _loaded_object))
                (bind _returnvalue (Rendering|Material|CreateDynamicMaterialInstance _asmaterial_interface "None" "None" :!class /Script/Engine.KismetMaterialLibrary))
                (bind _output_get (Variables|Default|SetLoadedDecalMID _returnvalue))
                (Rendering|Components|Decal|SetDecalMaterial (Variables|Default|GetDecalComponent) _output_get))))
          (if (Utilities|IsValidSoftObjectReference _decal_mask_largeicon)
            (bind _loaded_object_2 (Creative|Loading|AsyncLoadEditorSafe _decal_mask_largeicon)
              (:Loaded
                (bind _astexture (Utilities|Casting|CastToTexture _loaded_object_2))
                (Variables|Default|SetLoadedDecalTexture _astexture)))
            (else
              (Variables|Default|SetLoadedDecalTexture "")))
          (CallFunction|UpdateMIDParameters)
          ;[1 0.7382 0] Setting Color Component Data (will call 'Update Visuals on Client' when updated)
          (Class|CreativeColorComponent|SetInitialValues (Variables|Default|GetCreativeColorComponent) (Creative|TeamOption|GetLegacyTeamIndex (Variables|UserOptions|GetInitial%20Team%20Color)) (Utilities|Select :Index (Variables|UserOptions|GetColor%20Type) :NewEnumerator0 "NewEnumerator1" :NewEnumerator1 "NewEnumerator0" :NewEnumerator2 "NewEnumerator1" :NewEnumerator3 "NewEnumerator1") (* (Variables|UserOptions|GetCustom%20Color) 255.0) 1.0 2.0 0)
          (CallFunction|UpdateDecalScale))
        (:else
          _)))))

(event UserOptions|EventHandleUserOptionsChanged
  (params (ChangedPropertyNames :type Array%20of%20Names))
  (bind (_decal_mask_smallicon _decal_mask_largeicon) (Utilities|Struct|BreakCreativeIconOption (Variables|GetDecal%20Mask_SmallIcon)))
  (|CallParentFunction ChangedPropertyNames)
  (Networking|CanExecuteCosmeticEvents
    (:True
      ;[0.5 1 1] Getting the decal MID
      (bind _loadeddecalmid (Variables|Default|GetLoadedDecalMID)
        (:then
          (if (not (== _loadeddecalmid (Utilities|ResolveSoftReference (Variables|UserOptions|Basic|GetDecal%20Material))))
            (bind _loaded_object (Creative|Loading|AsyncLoadEditorSafe (Variables|UserOptions|Basic|GetDecal%20Material))
              (:Loaded
                (bind _asmaterial_interface (Utilities|Casting|CastToMaterialInterface _loaded_object))
                (bind _returnvalue (Rendering|Material|CreateDynamicMaterialInstance _asmaterial_interface "None" "None" :!class /Script/Engine.KismetMaterialLibrary))
                (bind _output_get (Variables|Default|SetLoadedDecalMID _returnvalue))
                (Rendering|Components|Decal|SetDecalMaterial (Variables|Default|GetDecalComponent) _output_get))))
          (if (Utilities|IsValidSoftObjectReference _decal_mask_largeicon)
            (bind _loaded_object_2 (Creative|Loading|AsyncLoadEditorSafe _decal_mask_largeicon)
              (:Loaded
                (bind _astexture (Utilities|Casting|CastToTexture _loaded_object_2))
                (Variables|Default|SetLoadedDecalTexture _astexture)))
            (else
              (Variables|Default|SetLoadedDecalTexture "")))
          (CallFunction|UpdateMIDParameters)
          ;[1 0.7382 0] Setting Color Component Data (will call 'Update Visuals on Client' when updated)
          (Class|CreativeColorComponent|SetInitialValues (Variables|Default|GetCreativeColorComponent) (Creative|TeamOption|GetLegacyTeamIndex (Variables|UserOptions|GetInitial%20Team%20Color)) (Utilities|Select :Index (Variables|UserOptions|GetColor%20Type) :NewEnumerator0 "NewEnumerator1" :NewEnumerator1 "NewEnumerator0" :NewEnumerator2 "NewEnumerator1" :NewEnumerator3 "NewEnumerator1") (* (Variables|UserOptions|GetCustom%20Color) 255.0) 1.0 2.0 0)
          (CallFunction|UpdateDecalScale))
        (:else
          _)))))

(event Custom|RefreshDecalInEditor
  (bind (_decal_mask_smallicon _decal_mask_largeicon) (Utilities|Struct|BreakCreativeIconOption (Variables|GetDecal%20Mask_SmallIcon)))
  (Networking|CanExecuteCosmeticEvents
    (:True
      ;[0.5 1 1] Getting the decal MID
      (bind _loadeddecalmid (Variables|Default|GetLoadedDecalMID)
        (:then
          (if (not (== _loadeddecalmid (Utilities|ResolveSoftReference (Variables|UserOptions|Basic|GetDecal%20Material))))
            (bind _loaded_object (Creative|Loading|AsyncLoadEditorSafe (Variables|UserOptions|Basic|GetDecal%20Material))
              (:Loaded
                (bind _asmaterial_interface (Utilities|Casting|CastToMaterialInterface _loaded_object))
                (bind _returnvalue (Rendering|Material|CreateDynamicMaterialInstance _asmaterial_interface "None" "None" :!class /Script/Engine.KismetMaterialLibrary))
                (bind _output_get (Variables|Default|SetLoadedDecalMID _returnvalue))
                (Rendering|Components|Decal|SetDecalMaterial (Variables|Default|GetDecalComponent) _output_get))))
          (if (Utilities|IsValidSoftObjectReference _decal_mask_largeicon)
            (bind _loaded_object_2 (Creative|Loading|AsyncLoadEditorSafe _decal_mask_largeicon)
              (:Loaded
                (bind _astexture (Utilities|Casting|CastToTexture _loaded_object_2))
                (Variables|Default|SetLoadedDecalTexture _astexture)))
            (else
              (Variables|Default|SetLoadedDecalTexture "")))
          (CallFunction|UpdateMIDParameters)
          ;[1 0.7382 0] Setting Color Component Data (will call 'Update Visuals on Client' when updated)
          (Class|CreativeColorComponent|SetInitialValues (Variables|Default|GetCreativeColorComponent) (Creative|TeamOption|GetLegacyTeamIndex (Variables|UserOptions|GetInitial%20Team%20Color)) (Utilities|Select :Index (Variables|UserOptions|GetColor%20Type) :NewEnumerator0 "NewEnumerator1" :NewEnumerator1 "NewEnumerator0" :NewEnumerator2 "NewEnumerator1" :NewEnumerator3 "NewEnumerator1") (* (Variables|UserOptions|GetCustom%20Color) 255.0) 1.0 2.0 0)
          (CallFunction|UpdateDecalScale))
        (:else
          _)))))

(event UpdateVisualsonClient(CreativeColorComponent)
  ;[1 0.5 1] Updating ColorLight on the MID
  (bind _loadeddecalmid (Variables|Default|GetLoadedDecalMID)
    (:then
      (Rendering|Material|SetVectorParameterValue _loadeddecalmid "ColorLight" (Class|CreativeColorComponentBlueprintLibrary|HSVAToRGBA (Class|CreativeColorComponent|GetAdjustedColor (Variables|Default|GetCreativeColorComponent))) :!class /Script/Engine.MaterialInstanceDynamic))
    (:else)))

(event Custom|RefreshColorInEditor
  ;[1 0.5 1] Updating ColorLight on the MID
  (bind _loadeddecalmid (Variables|Default|GetLoadedDecalMID)
    (:then
      (Rendering|Material|SetVectorParameterValue _loadeddecalmid "ColorLight" (Class|CreativeColorComponentBlueprintLibrary|HSVAToRGBA (Class|CreativeColorComponent|GetAdjustedColor (Variables|Default|GetCreativeColorComponent))) :!class /Script/Engine.MaterialInstanceDynamic))
    (:else)))

(fn GetCreativeActorBounds
  (bind (_origin _boxextent) (Collision|GetComponentBounds (Variables|Default|GetDecalComponent)))
  (return true _boxextent))

(fn GetCreativeActorOrigin
  (return true (Variables|/Script/Engine.SceneComponent|GetRelativeLocation (Variables|Default|GetDecalComponent))))

(fn GetCreativeActorRelativeTransform
  (return false 0))

(fn GetGridAxisGuideAttachmentPoint
  (return false false "0, 0, 0"))

(fn MigrateLegacyUserOptions
  (bind _knob_color (Variables|Legacy|GetKnob_Color))
  (bind _returnvalue (Utilities|Enum|Equal(Enum) _knob_color "NewEnumerator1"))
  (bind (_decal_material _knob_color_2) (Class|GetClassDefaults "/CreativeCoreDevices/Device_Decal_V2.Device_Decal_V2_C"))
  (bind (_decal_material_2 _knob_color_3) (Class|GetClassDefaults "/CreativeCoreDevices/Device_Decal_V2.Device_Decal_V2_C"))
  (|CallParentFunction)
  (Utilities|FlowControl|Sequence
    (:then_0
      (Utilities|FlowControl|Sequence
        (:then_0
          ;[1 0 0.8044 0.2] Migrating Color option.
          (if (Utilities|Enum|NotEqual(Enum) _knob_color_3 _knob_color)
            (Variables|UserOptions|SetColor%20Type (select _returnvalue "NewEnumerator2" "NewEnumerator1"))
            (if (not _returnvalue)
              (Variables|UserOptions|SetCustom%20Color (Class|CreativeColorComponent|GetColorKeyAsRGB (Variables|Default|GetCreativeColorComponent) (Variables|Legacy|GetKnob_Color)))
              (else
                _))
            (Variables|Legacy|SetKnob_Color _knob_color_2)
            (LocalVariables|SetOption%20migrated true)))
        (:then_1
          ;[1 1 1] Team could be None (Invalid now) due to old settings.
          (if (Creative|TeamOption|IsTeamType (Variables|UserOptions|GetInitial%20Team%20Color) "None")
            (Variables|SetInitial%20Team%20Color_TeamType (struct :TeamType "Any" :TeamIndex 1))
            (LocalVariables|SetOption%20migrated true)))))
    (:then_1
      (if (LocalVariables|GetOption%20migrated)
        (OptionData|RefreshPropertyOverrides (Variables|Default|GetToyOptionsComponent))))))

(fn UpdateDecalScale
  (bind _decal_distance (Variables|UserOptions|GetDecal%20Distance))
  (bind _returnvalue (* _decal_distance 5e+01))
  (bind _decalcomponent (Variables|Default|GetDecalComponent))
  (bind (_relativescale3d_x _relativescale3d_y _relativescale3d_z) (Math|Vector|BreakVector (Variables|/Script/Engine.SceneComponent|GetRelativeScale3D_X _decalcomponent)))
  (Transformation|SetRelativeLocation _decalcomponent (Math|Vector|MakeVector _returnvalue) false false)
  (Transformation|SetRelativeScale3D _decalcomponent (Math|Vector|MakeVector _returnvalue _relativescale3d_y _relativescale3d_z)))

(fn UpdateMIDParameters
  (bind _loadeddecaltexture (Variables|Default|GetLoadedDecalTexture))
  (bind _loadeddecalmid (Variables|Default|GetLoadedDecalMID))
  (Rendering|Material|SetTextureParameterValue _loadeddecalmid "DecalTexture" _loadeddecaltexture)
  (Rendering|Material|SetTextureParameterValue _loadeddecalmid "ShineMask" _loadeddecaltexture)
  (Rendering|Material|SetScalarParameterValue _loadeddecalmid "Opacity" (Math|Float|SafeDivide (Variables|UserOptions|GetOpacity) 1e+02) :!class /Script/Engine.MaterialInstanceDynamic))

(fn ConstructionScript
  (if (Editor|InEditor)
    (Networking|CanExecuteCosmeticEvents
      (:True
        (CallFunction|RefreshDecalinEditor)
        (CallFunction|RefreshColorinEditor)))))
