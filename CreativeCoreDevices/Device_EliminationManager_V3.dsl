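; Creative Elimination Manager device (V3) that reacts to qualifying player and AI eliminations by broadcasting events and queuing item spawns at death locations.
; Pickups hitting the DropBox outside play mode register in the spawner component's item list.
; Server-side death bindings use controller components and the AI population tracker, filtered by target tags, with team/class restrictions for players.
; It also applies optional player material/gold/score penalties, toggles VFX tags and migrates legacy options.
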(event EventInitializeDevice
  ;[0.1134 1 0] Initialization
  (|CallParentFunction)
  (Initialization|InitializeComponent (Variables|Default|GetCreative_ItemPreview_Component) (Variables|Default|GetRotationSceneComponent) (Variables|Default|GetPreviewStaticMesh) (Variables|Default|GetPreviewSkeletalMesh) 0 0 (Variables|Default|GetItemVariantNotifier) :!class /Game/Creative/Devices/Common/Components/Creative_ItemPreview_Component.Creative_ItemPreview_Component_C)
  (CallFunction|SetupMinigameSpawner)
  (Initialization|Initialize (Variables|Default|GetCreative_UseRestriction_Component) (Variables|Default|GetFortMinigameProgress) 0 0 (Variables|UserOptions|Basic|GetSelected%20Team) (Variables|UserOptions|GetInvert%20Selected%20Team) (Variables|UserOptions|Basic|GetSelected%20Class) (Variables|UserOptions|GetInvert%20Selected%20Class) false)
  (CallFunction|ResaveActor))

(event EventOnPlayModeChanged
  (params (Minigame :type Fort%20Minigame%20Object%20Reference) (bIsInPlayMode :type Boolean))
  ;[1 0.1082 0 0.25] Disable Collision on drop area during Play mode, as overlaps are expensive.
  (|CallParentFunction Minigame bIsInPlayMode)
  (Collision|SetCollisionEnabled (Variables|Default|GetDropBox) (select bIsInPlayMode "NoCollision" "QueryOnly"))
  (Default|Reset))

(event EventBeginPlay
  ;[0.1134 1 0] Initialization
  (|CallParentFunction)
  (Variables|Default|SetCachedMinigameStatsComponent (Minigame|Stats|GetMinigameStatsComponent self))
  (Networking|SwitchHasAuthority
    (:Authority
      ;[0 1 0.1504] Binding dispatcher events from the Spawner Component. Done here to be server only.
      (bind _minigame_spawner_component (Variables|Default|GetMinigame_Spawner_Component)
        (:then
          (Default|BindEventtoAddedtoSpawnListDispatcher _minigame_spawner_component (delegate Custom|OnAddedToSpawnList))
          (Default|BindEventtoOnSpawnDispatcher _minigame_spawner_component (EventDispatchers|CreateEvent :!event Set%20Spawned%20Item%20Settings))
          (Default|BindEventtoOnPickedUpSpawnedItem _minigame_spawner_component (delegate Custom|Picked%20Up%20Item)))
        (:else)))))

(event Custom|OnAddedToSpawnList
  (params (Added%20Actor :type Actor%20Object%20Reference))
  ;[0 1 0.1504] Binding dispatcher events from the Spawner Component. Done here to be server only.
  (CallFunction|ResaveActor))

(event Custom|Picked%20Up%20Item
  (params (Controller :type Controller%20Object%20Reference))
  ;[0 1 0.1504] Binding dispatcher events from the Spawner Component. Done here to be server only.
  (GameplayEvent|BroadcastEvent (Variables|UserOptions|Events|GetOn%20Item%20Picked%20Up) 0 (Utilities|Struct|MakeGameplayEventItemSpawnerPickedUpItem Controller)))

(event OnItemContainerUpdated(Minigame_Spawner_Component)
  ;[0.1134 1 0] Initialization
  (CallFunction|UpdateItemPreviewList))

(event MinigameStartedDelegate(FortMinigameProgress)
  ;[0 1 0.175] When minigame starts
  (CallFunction|HandleVFXAll true)
  (Networking|SwitchHasAuthority
    (:Authority
      (CallFunction|InitializeEliminationListening)
      (Variables|Default|SetbDisabledSpawning (not (Variables|UserOptions|GetEnabled%20At%20Game%20Start)))
      (Default|TryOverrideWeaponAmmoonSpawnParams (Variables|Default|GetMinigame_Spawner_Component) (Variables|UserOptions|GetInitialWeaponAmmo_Override) (Variables|UserOptions|GetInitialWeaponAmmo))
      (CallFunction|BindMinigameEvents true))))

(event MinigameEndedDelegate(FortMinigameProgress)
  ;[1 1 1] When Minigame Ends
  (CallFunction|HandleVFXAll false)
  (CallFunction|BindMinigameEvents false))

(event Custom|PlayerChangedTeam
  (params (PlayerState :type Fort%20Player%20State%20Object%20Reference) (TeamId :type Byte) (OldTeamId :type Byte) (bTeamChangedWithoutRespawn :type Boolean))
  (bind _returnvalue (Fortnite|GetControllerfromActor PlayerState))
  ;[1 1 1] Minigame Player Bindings
  (Utilities|IsValid _returnvalue
    (:Is%20Valid
      (CallFunction|HandleVFX _returnvalue (CreativeUserRestrictions|CanParticipantActorInteract (Variables|Default|GetCreative_UseRestriction_Component) PlayerState false true false)))))

(event Custom|PlayerChangedClass
  (params (PlayerState :type Fort%20Player%20State%20Object%20Reference) (ClassSlotIndex :type Byte))
  (bind _returnvalue (Fortnite|GetControllerfromActor PlayerState))
  ;[1 1 1] Minigame Player Bindings
  (Utilities|IsValid _returnvalue
    (:Is%20Valid
      (CallFunction|HandleVFX _returnvalue (CreativeUserRestrictions|CanParticipantActorInteract (Variables|Default|GetCreative_UseRestriction_Component) PlayerState false true false)))))

(event Custom|PlayerRemoved
  (params (UniqueNetId :type Unique%20Net%20Id%20Structure) (bIsLocalPlayer :type Boolean))
  (bind _returnvalue (Class|CreativeBlueprintLibrary|GetPlayerStatefromUniqueNetID UniqueNetId))
  (bind _returnvalue_2 (Fortnite|GetControllerfromActor _returnvalue))
  ;[1 1 1] Minigame Player Bindings
  (Utilities|IsValid _returnvalue_2
    (:Is%20Valid
      (CallFunction|HandleVFX _returnvalue_2))))

(event Custom|OnPlayerFinishedAddedToMinigame
  (params (UniqueNetId :type Unique%20Net%20Id%20Structure) (bIsLocalPlayer :type Boolean) (FortPlayerState :type Fort%20Player%20State%20Object%20Reference))
  ;[1 1 1] Minigame Player Bindings
  (Utilities|IsValid FortPlayerState
    (:Is%20Valid
      (CallFunction|HandleVFX (PlayerState|GetPlayerController FortPlayerState) true))))

(event OnComponentHit(DropBox)
  (params (HitComponent :type Primitive%20Component%20Object%20Reference) (OtherActor :type Actor%20Object%20Reference) (OtherComp :type Primitive%20Component%20Object%20Reference) (NormalImpulse :type Vector :default "0, 0, 0") (Hit :type Hit%20Result%20Structure%20%28by%20ref%29))
  (bind _itemstobeaddedqueue (Variables|Default|GetItemsToBeAddedQueue))
  (bind _itemstobeaddedqueue_2 (Variables|Default|GetItemsToBeAddedQueue))
  ;[0 0.5614 1 0.2] Handles item being added from a hit event.  Hit captures all pick-ups
  (Networking|SwitchHasAuthority
    (:Authority
      (if (not (Default|IsinPlayMode))
        (bind _output_get (Variables|Default|SetItemThatHit OtherActor))
        (bind _asfort_pickup_athena (Utilities|Casting|CastToFortPickupAthena _output_get))
        (Utilities|IsPickupValid _asfort_pickup_athena (Variables|Default|GetInvalidPickupGameplayTagContainer)
          (:IsValid
            (if (Interaction|BlueprintCanInteract (Pickup|GetPawnWhoDroppedPickup _asfort_pickup_athena))
              ;;[1 1 1] Every bounce of the pickup causes a hit trigger.  This ensures rest of script only fires once
              (if (not (Utilities|Equal(Object) (Variables|Default|GetItemThatHit) (Variables|Default|GetItemToBeAdded)))
                (bind _output_get_2 (Variables|Default|SetItemToBeAdded (Variables|Default|GetItemThatHit)))
                (bind _returnvalue (Utilities|Array|Add _itemstobeaddedqueue _output_get_2))
                (CallFunction|RPCMULTICASTAddItemFX (Utilities|Array|Get(acopy) _itemstobeaddedqueue _returnvalue))
                (Utilities|FlowControl|Delay 2.0) ; Ensure object still exists so FX can start before object is destroyed
                ;[1 1 1] Loop through Queue array to be sure to add every item if items are added in rapid succession
                (if (> (Utilities|Array|Length _itemstobeaddedqueue_2) 0)
                  (CallFunction|AddtoSpawnList (Utilities|Array|Get(acopy) _itemstobeaddedqueue_2)))))))))))

(event Custom|Enable%20Event
  (params (Instigator :type Controller%20Object%20Reference))
  ;[0.9614 0 1] Gameplay Functions
  (Networking|SwitchHasAuthority
    (:Authority
      (bind _binprogress (CallFunction|IsMinigameInProgress))
      (if _binprogress
        (if (Variables|Default|GetbDisabledSpawning)
          (Variables|Default|SetbDisabledSpawning false))))))

(event Custom|Disable%20Event
  (params (Instigator :type Controller%20Object%20Reference))
  ;[0.9614 0 1] Gameplay Functions
  (Networking|SwitchHasAuthority
    (:Authority
      (bind _binprogress (CallFunction|IsMinigameInProgress))
      (if _binprogress
        (if (not (Variables|Default|GetbDisabledSpawning))
          (Variables|Default|SetbDisabledSpawning true))))))

(event OnLastItemPickedUp(Minigame_Spawner_Component)
  (Networking|SwitchHasAuthority
    (:Authority
      (if (> (Utilities|Array|Length (Variables|Default|GetQueuedSpawnLocations)) 0)
        (Utilities|Array|RemoveIndex (Variables|Default|GetQueuedSpawnLocations))
        (bind _minigame_spawner_component (Variables|Default|GetMinigame_Spawner_Component)
          (:then
            (Class|MinigameSpawnerComponent|ResetSpawner _minigame_spawner_component false)
            (if (> (Utilities|Array|Length (Variables|Default|GetQueuedSpawnLocations)) 0)
              (Class|MinigameSpawnerComponent|SpawnRemaining (Variables|Default|GetMinigame_Spawner_Component) true)))
          (:else))))))

(event Custom|CleanUpMinigameDelegates
  (bind _fortminigameprogress (Variables|Default|GetFortMinigameProgress)
    (:then
      (bind _currentminigame (Variables|/Script/FortniteGame.FortMinigameLogicComponent|GetCurrentMinigame _fortminigameprogress)
        (:then
          (Minigame|UnbindEventfromPlayerRemovedDelegate _currentminigame (EventDispatchers|CreateEvent :!event Custom|OnPlayerRemoved))
          (Minigame|UnbindEventfromOnPlayerJoininProgress _currentminigame (EventDispatchers|CreateEvent :!event Custom|OnPlayerJoinedInProgress))
          (Minigame|UnbindEventfromOnAddedtoMinigame _fortminigameprogress (EventDispatchers|CreateEvent :!event OnAddedToMinigame))
          (bind _returnvalue (AI|GetAIPopulationTracker))
          (Utilities|IsValid _returnvalue
            (:Is%20Valid
              (AI|UnbindEventfromOnAIPawnDied _returnvalue (EventDispatchers|CreateEvent :!event Custom|OnAIPawnDied)))))
        (:else)))
    (:else)))

(event Custom|Initialize%20Elimination%20Listening
  (Networking|SwitchHasAuthority
    (:Authority
      (bind _fortminigameprogress (Variables|Default|GetFortMinigameProgress)
        (:then
          (bind _currentminigame (Variables|/Script/FortniteGame.FortMinigameLogicComponent|GetCurrentMinigame _fortminigameprogress)
            (:then
              (Utilities|FlowControl|Sequence
                (:then_0
                  (bind _outplayers (Minigame|GetParticipatingPlayers_NotPure _currentminigame))
                  (for _array_element _outplayers
                    (Utilities|IsValid _array_element
                      (:Is%20Valid
                        (Utilities|FlowControl|Sequence
                          (:then_0
                            (Utilities|IsValid (Fortnite|GetControllerfromActor _array_element)
                              (:Is%20Valid
                                (Pawn|BindEventtoOnPossessedPawnChanged (Fortnite|GetControllerfromActor _array_element) (EventDispatchers|CreateEvent :!event Custom|OnPlayerPawnPossessed)))))
                          (:then_1
                            (CallFunction|BindtoPawnDeath (PlayerState|GetCurrentPawn _array_element))))))))
                (:then_1
                  (Minigame|AssignOnPlayerJoininProgress _currentminigame (EventDispatchers|CreateEvent :!event Custom|OnPlayerJoinedInProgress)))
                (:then_2
                  (Minigame|AssignPlayerRemovedDelegate _currentminigame (EventDispatchers|CreateEvent :!event Custom|OnPlayerRemoved)))
                (:then_3
                  (bind _returnvalue (AI|GetAIPopulationTracker))
                  (Utilities|IsValid _returnvalue
                    (:Is%20Valid
                      (AI|BindEventtoOnAIPawnDied _returnvalue (EventDispatchers|CreateEvent :!event Custom|OnAIPawnDied)))))))
            (:else)))
        (:else)))))

(event Custom|CleanUpPlayerDelegates
  (params (PlayerState :type Fort%20Player%20State%20Object%20Reference))
  (bind _returnvalue (Fortnite|GetControllerfromActor PlayerState))
  (Utilities|IsValid PlayerState
    (:Is%20Valid
      (Utilities|FlowControl|Sequence
        (:then_0
          (Utilities|IsValid _returnvalue
            (:Is%20Valid
              (Pawn|UnbindEventfromOnPossessedPawnChanged _returnvalue (EventDispatchers|CreateEvent :!event Custom|OnPlayerPawnPossessed)))))
        (:then_1
          (CallFunction|BindtoPawnDeath (Default|GetPlayerPawnAthenafromPlayerState PlayerState)))))))

(event OnLastItemFinishedSpawning(Minigame_Spawner_Component)
  (Networking|SwitchHasAuthority
    (:Authority
      (if (> (Utilities|Array|Length (Variables|Default|GetQueuedSpawnLocations)) 0)
        (Utilities|Array|RemoveIndex (Variables|Default|GetQueuedSpawnLocations))
        (bind _minigame_spawner_component (Variables|Default|GetMinigame_Spawner_Component)
          (:then
            (Class|MinigameSpawnerComponent|ResetSpawner _minigame_spawner_component false)
            (if (> (Utilities|Array|Length (Variables|Default|GetQueuedSpawnLocations)) 0)
              (Class|MinigameSpawnerComponent|SpawnRemaining (Variables|Default|GetMinigame_Spawner_Component) true)))
          (:else))))))

(event Custom|OnPlayerPawnPossessed
  (params (OldPawn :type Pawn%20Object%20Reference) (NewPawn :type Pawn%20Object%20Reference))
  (CallFunction|BindtoPawnDeath NewPawn))

(event Custom|OnPlayerRemoved
  (params (UniqueNetId :type Unique%20Net%20Id%20Structure) (bIsLocalPlayer :type Boolean))
  (CallFunction|CleanUpPlayerDelegates (Default|GetFortPlayerStatebyUniqueNetId UniqueNetId)))

(event Custom|OnPlayerJoinedInProgress
  (params (Pawn :type Pawn%20Object%20Reference))
  (bind _returnvalue (Pawn|GetController Pawn))
  (Utilities|FlowControl|Sequence
    (:then_0
      (CallFunction|BindtoPawnDeath Pawn))
    (:then_1
      (Utilities|IsValid _returnvalue
        (:Is%20Valid
          (Pawn|BindEventtoOnPossessedPawnChanged _returnvalue (EventDispatchers|CreateEvent :!event Custom|OnPlayerPawnPossessed)))))))

(event Custom|OnAIPawnDied
  (params (KilledController :type Controller%20Object%20Reference) (bIsABot :type Boolean) (KilledAIPawn :type Fort%20Pawn%20Object%20Reference) (KilledPlayerPawn :type Fort%20Player%20Pawn%20Object%20Reference) (KillerController :type Controller%20Object%20Reference))
  (Utilities|IsValid KilledPlayerPawn
    (:Is%20Valid
      (if bIsABot
        (Default|HandlePawnDeath KilledPlayerPawn KillerController)))
    (:Is%20Not%20Valid
      (Default|HandlePawnDeath KilledAIPawn KillerController))))

(event Custom|RPC_MULTICAST%20Add%20Item%20FX
  (params (Item :type Actor%20Object%20Reference))
  (Networking|CanExecuteCosmeticEvents
    (:True
      (bind _asfort_pickup_athena (Utilities|Casting|CastToFortPickupAthena Item))
      (Class|CreativeItemPreviewComponent|ItemAddedToQueueFX (Variables|Default|GetCreative_ItemPreview_Component) _asfort_pickup_athena))))

(fn AddToSpawnList
  (params (InputPin :type Actor%20Object%20Reference))
  (Networking|SwitchHasAuthority
    (:Authority
      (bind _wasvalid (Class|MinigameSpawnerComponent|AddtoSpawnList (Variables|Default|GetMinigame_Spawner_Component) InputPin (Variables|Default|GetItemSpawnLocation)))
      (if _wasvalid
        (Actor|DestroyActor InputPin) ; Destroy the pick-up that is in the world.  Hidden by FX
        (else
          _))
      (Utilities|Array|RemoveIndex (Variables|Default|GetItemsToBeAddedQueue)))))

(fn BindMinigameEvents
  (params (Bind :type Boolean))
  (bind _outputdelegate (EventDispatchers|CreateEvent :!event PlayerChangedTeam))
  (bind _outputdelegate_2 (EventDispatchers|CreateEvent :!event PlayerChangedClass))
  (bind _outputdelegate_3 (EventDispatchers|CreateEvent :!event PlayerRemoved))
  (bind _outputdelegate_4 (EventDispatchers|CreateEvent :!event OnPlayerFinishedAddedToMinigame))
  (bind _minigameref (Variables|Default|GetMinigameRef))
  (bind _minigameref_2 (Variables|Default|GetMinigameRef))
  (bind _minigameref_3 (Variables|Default|GetMinigameRef))
  (bind _minigameref_4 (Variables|Default|GetMinigameRef))
  (bind _currentminigame (Variables|/Script/FortniteGame.FortMinigameLogicComponent|GetCurrentMinigame (Variables|Default|GetFortMinigameProgress))
    (:then
      (Variables|Default|SetMinigameRef _currentminigame)
      (if (Variables|GetBind)
        (Minigame|BindEventtoPlayerChangedTeamDelegate _minigameref_3 _outputdelegate)
        (Minigame|BindEventtoOnPlayerClassApplied _minigameref _outputdelegate_2)
        (Minigame|BindEventtoPlayerRemovedDelegate _minigameref_4 _outputdelegate_3)
        (Minigame|BindEventtoOnFinishedAddingMinigamePlayerDelegate _minigameref_2 _outputdelegate_4)
        (else
          (Minigame|UnbindEventfromPlayerChangedTeamDelegate _minigameref_3 _outputdelegate)
          (Minigame|UnbindEventfromOnPlayerClassApplied _minigameref _outputdelegate_2)
          (Minigame|UnbindEventfromPlayerRemovedDelegate _minigameref_4 _outputdelegate_3)
          (Minigame|UnbindEventfromOnFinishedAddingMinigamePlayerDelegate _minigameref_2 _outputdelegate_4))))
    (:else)))

(fn BindToPawnDeath
  (params (Pawn :type Pawn%20Object%20Reference))
  (bind _asfort_player_pawn_athena (Utilities|Casting|CastToFortPlayerPawnAthena Pawn)
    (:then
      (bind _returnvalue (Components|AddControllerComponent (Pawn|GetController _asfort_player_pawn_athena) "/Game/Creative/Devices/EliminationManager/BP_FortControllerComponent_EliminationManagement.BP_FortControllerComponent_EliminationManagement_C"))
      (bind _asbp_fort_controller_component_elimination_management (Utilities|Casting|CastToBP_FortControllerComponent_EliminationManagement _returnvalue)
        (:then
          (Default|UnbindEventfromOnPawnDeath _asbp_fort_controller_component_elimination_management (EventDispatchers|CreateEvent :!event OnPawnDeath))
          (Default|BindEventtoOnPawnDeath _asbp_fort_controller_component_elimination_management (EventDispatchers|CreateEvent :!event OnPawnDeath))
          (return true))
        (:CastFailed
          (return false))))
    (:CastFailed
      (return false))))

(fn CleanUpPlayers
  (bind _fortminigameprogress (Variables|Default|GetFortMinigameProgress)
    (:then
      (bind _currentminigame (Variables|/Script/FortniteGame.FortMinigameLogicComponent|GetCurrentMinigame _fortminigameprogress)
        (:then
          (for _array_element (Minigame|GetParticipatingPlayers _currentminigame)
            (CallFunction|CleanUpPlayerDelegates _array_element)))
        (:else)))
    (:else)))

(fn DroporRemoveMaterials
  (params (Eliminated%20Controller :type Controller%20Object%20Reference) (DropParams :type Fort%20Creative%20Item%20Drop%20Params%20Structure))
  (bind _itemremovedcount (Inventory|DropSpecificItem :self (Variables|Default|GetFortCreativeInventoryManip) :Controller (Variables|GetEliminated%20Controller) :ItemDefinition "/Game/Items/ResourcePickups/WoodItemData.WoodItemData" :DropParams (Variables|GetDropParams) :!class /Script/FortniteGame.FortCreativeInventoryManipComponent))
  (bind _itemremovedcount_2 (Inventory|DropSpecificItem :self (Variables|Default|GetFortCreativeInventoryManip) :Controller (Variables|GetEliminated%20Controller) :ItemDefinition "/Game/Items/ResourcePickups/StoneItemData.StoneItemData" :DropParams (Variables|GetDropParams) :!class /Script/FortniteGame.FortCreativeInventoryManipComponent))
  (bind _itemremovedcount_3 (Inventory|DropSpecificItem :self (Variables|Default|GetFortCreativeInventoryManip) :Controller (Variables|GetEliminated%20Controller) :ItemDefinition "/Game/Items/ResourcePickups/MetalItemData.MetalItemData" :DropParams (Variables|GetDropParams) :!class /Script/FortniteGame.FortCreativeInventoryManipComponent))
  (return _itemremovedcount_3 _itemremovedcount_2 _itemremovedcount))

(fn GetFinisherOrDownerPawn
  (params (DyingPlayerPawn :type Fort%20Player%20Pawn%20Athena%20Object%20Reference))
  (bind _playerstate (Variables|/Script/Engine.Pawn|GetPlayerState DyingPlayerPawn)
    (:then
      (bind _asfort_player_state_athena (Utilities|Casting|CastToFortPlayerStateAthena _playerstate))
      (bind _returnvalue_finisherordowner (Utilities|Struct|BreakDeathInfo (Class|FortPlayerStateAthena|GetDeathInfo _asfort_player_state_athena)))
      (Utilities|IsValid _returnvalue_finisherordowner
        (:Is%20Valid
          (Utilities|IsValid (Fortnite|GetControllerfromActor _returnvalue_finisherordowner)
            (:Is%20Valid
              (bind _asfort_player_pawn_athena (Utilities|Casting|CastToFortPlayerPawnAthena (Pawn|GetControlledPawn (Fortnite|GetControllerfromActor _returnvalue_finisherordowner))))
              (return _asfort_player_pawn_athena))))))
    (:else)))

(fn GetFortPlayerStateByUniqueNetId
  (params (UniqueNetId :type Unique%20Net%20Id%20Structure))
  (bind _returnvalue (Game|GetGameState))
  (Utilities|FlowControl|Sequence
    (:then_0
      (Utilities|IsValid _returnvalue
        (:Is%20Valid
          (for _array_element (Variables|/Script/Engine.GameStateBase|GetPlayerArray _returnvalue)
            (bind _asfort_player_state_athena (Utilities|Casting|CastToFortPlayerStateAthena _array_element))
            (if (Online|UniqueNetIdRepl|Equal(UniqueIdRepl) (PlayerState|GetPlayerUniqueNetID _asfort_player_state_athena) UniqueNetId)
              (LocalVariables|SetPlayerState _asfort_player_state_athena))))))
    (:then_1
      (return (LocalVariables|GetPlayerState)))))

(fn GetPlayerPawnAthenaFromPlayerState
  (params (PlayerState :type Fort%20Player%20State%20Object%20Reference))
  (Utilities|FlowControl|Sequence
    (:then_0
      (Utilities|IsValid PlayerState
        (:Is%20Valid
          (bind _asfort_player_pawn_athena (Utilities|Casting|CastToFortPlayerPawnAthena (PlayerState|GetCurrentPawn PlayerState)))
          (LocalVariables|SetFortPlayerPawnAthena _asfort_player_pawn_athena))))
    (:then_1
      (return (LocalVariables|GetFortPlayerPawnAthena)))))

(fn GetRandomSpawnOffset
  (bind (_returnvalue_x _returnvalue_y) (Math|Vector|BreakVector (Math|Random|RandomUnitVector)))
  ;[0 1 0.9331 0.25] Returns a random 2d Offset in Vector Form. This explicitly uses a 2d unit vector as its base to ensure that the random distribution is in a circle rather than a square (which is what would happen if you just generated two random floats, which is what the original elimination manager did).
  (if (> (Variables|UserOptions|GetRandom%20Spawn%20Distance) 0)
    (bind _output_get (LocalVariables|SetRandom%20Distance (Math|Random|RandomFloatinRange 0.0 (* (Variables|UserOptions|GetRandom%20Spawn%20Distance) 1e+02))))
    (bind (_returnvalue_x_2 _returnvalue_y_2) (Math|Vector2D|BreakVector2D (Utilities|Operators|Multiply (Math|Vector2D|Normalize2D (Math|Vector2D|MakeVector2D _returnvalue_x _returnvalue_y)) (Math|Vector2D|MakeVector2D _output_get _output_get))))
    (return _returnvalue_x_2 _returnvalue_y_2 0.0)
    (else
      (return "0, 0, 0"))))

(fn GrantItemsToPlayer
  (params (InventoryOwner :type Fort%20Inventory%20Owner%20Interface%20Interface) (GoldAmount :type Integer) (MetalAmount :type Integer) (StoneAmount :type Integer) (WoodAmount :type Integer))
  (Inventory|GiveItemtoInventoryOwner :InventoryOwner InventoryOwner :ItemDefinition "/Game/Items/ResourcePickups/GoldItemData.GoldItemData" :NumberToGive GoldAmount :bNotifyPlayer true :ItemLevel -1 :PickupInstigatorHandle -1 :bUseItemPickupAnalyticEvent false :WeaponAmmoOverride -1)
  (Inventory|GiveItemtoInventoryOwner :InventoryOwner InventoryOwner :ItemDefinition "/Game/Items/ResourcePickups/MetalItemData.MetalItemData" :NumberToGive MetalAmount :bNotifyPlayer true :ItemLevel -1 :PickupInstigatorHandle -1 :bUseItemPickupAnalyticEvent false :WeaponAmmoOverride -1)
  (Inventory|GiveItemtoInventoryOwner :InventoryOwner InventoryOwner :ItemDefinition "/Game/Items/ResourcePickups/StoneItemData.StoneItemData" :NumberToGive StoneAmount :bNotifyPlayer true :ItemLevel -1 :PickupInstigatorHandle -1 :bUseItemPickupAnalyticEvent false :WeaponAmmoOverride -1)
  (Inventory|GiveItemtoInventoryOwner :InventoryOwner InventoryOwner :ItemDefinition "/Game/Items/ResourcePickups/WoodItemData.WoodItemData" :NumberToGive WoodAmount :bNotifyPlayer true :ItemLevel -1 :PickupInstigatorHandle -1 :bUseItemPickupAnalyticEvent false :WeaponAmmoOverride -1))

(fn HandleEliminationPenalties
  (params (Eliminated%20Controller :type Controller%20Object%20Reference) (Eliminator%20Controller :type Controller%20Object%20Reference))
  (bind _penaltyamount (Variables|UserOptions|GetPenaltyAmount))
  (bind _returnvalue (/ _penaltyamount 1e+02))
  (bind _eliminatedplayerscore (LocalVariables|GetEliminatedPlayerScore))
  (bind _returnvalue_2 (* _eliminatedplayerscore _returnvalue))
  (bind _returnvalue_3 (Math|Float|Truncate _returnvalue_2))
  (bind _removedgold (LocalVariables|GetRemovedGold))
  (bind _removedmetal (LocalVariables|GetRemovedMetal))
  (bind _removedstone (LocalVariables|GetRemovedStone))
  (bind _removedwood (LocalVariables|GetRemovedWood))
  (bind _eliminator_controller (Variables|GetEliminator%20Controller))
  (bind _cachedminigamestatscomponent (Variables|Default|GetCachedMinigameStatsComponent)
    (:then
      (LocalVariables|SetEliminatedPlayerScore (Minigame|Stats|GetStatforPlayer _cachedminigamestatscomponent (Utilities|Casting|CastToFortPlayerState (Variables|/Script/Engine.Controller|GetPlayerState Eliminated%20Controller)) (Class|FortMinigameStatsComponent|GetLoadedStatfromClass _cachedminigamestatscomponent "/Game/Athena/Playset/Minigames/StatFilters/MinigameStat_Score.MinigameStat_Score_C") false))
      (switch Utilities|FlowControl|Switch|SwitchonPenaltyEffectOption (Variables|UserOptions|GetPenaltyEffect)
        (:NewEnumerator0
          (LocalVariables|SetDropParams (Utilities|Struct|MakeFortCreativeItemDropParams "Percentage" 0 (/ (Variables|UserOptions|GetPenaltyAmount) 1e+02) false)))
        (:NewEnumerator1
          (LocalVariables|SetDropParams (Utilities|Struct|MakeFortCreativeItemDropParams "Percentage" 0 (/ (Variables|UserOptions|GetPenaltyAmount) 1e+02))))
        (:NewEnumerator2
          (LocalVariables|SetTransferItems true)
          (LocalVariables|SetDropParams (Utilities|Struct|MakeFortCreativeItemDropParams "Percentage" 0 (/ (Variables|UserOptions|GetPenaltyAmount) 1e+02) false))))
      (Utilities|FlowControl|Sequence
        (:then_0
          (switch Utilities|FlowControl|Switch|SwitchonPenaltyItemOption (Variables|UserOptions|GetPenalty%20Item)
            (:NewEnumerator0
              (Utilities|FlowControl|Sequence
                (:then_0
                  _)
                (:then_1
                  (bind (_removedmetal_2 _removedstone_2 _removedwood_2) (Default|DropOrRemoveMaterials (Variables|GetEliminated%20Controller) (LocalVariables|GetDropParams)))
                  (LocalVariables|SetRemovedMetal _removedmetal_2)
                  (LocalVariables|SetRemovedStone _removedstone_2)
                  (LocalVariables|SetRemovedWood _removedwood_2))
                (:then_2
                  (LocalVariables|SetRemovedScore _returnvalue_3)
                  (Class|CreativeCommonDeviceFunctionLibrary|AwardScore self _returnvalue_3 "NewEnumerator2" (Variables|GetEliminated%20Controller) (Variables|Default|GetFortMinigameProgress)))))
            (:NewEnumerator1
              _)
            (:NewEnumerator2
              _)
            (:NewEnumerator3
              _))
          (bind _itemremovedcount (Inventory|DropSpecificItem :self (Variables|Default|GetFortCreativeInventoryManip) :Controller (Variables|GetEliminated%20Controller) :ItemDefinition "/Game/Items/ResourcePickups/GoldItemData.GoldItemData" :DropParams (LocalVariables|GetDropParams) :!class /Script/FortniteGame.FortCreativeInventoryManipComponent))
          (LocalVariables|SetRemovedGold _itemremovedcount))
        (:then_1
          (if (LocalVariables|GetTransferItems)
            (Utilities|IsValid (Variables|GetEliminator%20Controller)
              (:Is%20Valid
                (Class|CreativeCommonDeviceFunctionLibrary|AwardScore self (LocalVariables|GetRemovedScore) "NewEnumerator1" (Variables|GetEliminator%20Controller) (Variables|Default|GetFortMinigameProgress))
                (bind _asfort_player_controller (Utilities|Casting|CastToFortPlayerController _eliminator_controller)
                  (:then
                    (Default|GrantItemstoPlayer _asfort_player_controller _removedgold _removedmetal _removedstone _removedwood))
                  (:CastFailed
                    (bind _asfort_athena_aibot_controller (Utilities|Casting|CastToFortAthenaAIBotController _eliminator_controller))
                    (Default|GrantItemstoPlayer _asfort_athena_aibot_controller _removedgold _removedmetal _removedstone _removedwood)))))))))
    (:else)))

(fn HandlePawnDeath
  (params (Killed%20AI%20Pawn :type Fort%20Pawn%20Object%20Reference) (Killer%20Controller :type Controller%20Object%20Reference))
  (Utilities|FlowControl|Sequence
    (:then_0
      (Utilities|IsValid Killed%20AI%20Pawn
        (:Is%20Valid
          (Fortnite|FortShippingLog self (Utilities|String|Append "Elimination Manager: AI Pawn eliminated: " (Utilities|GetDisplayName Killed%20AI%20Pawn)))
          (if (CallFunction|IsParticipatingPlayer Killed%20AI%20Pawn)
            (Fortnite|FortShippingLog self "Elimination Manager: AI Pawn is participating player. Ignoring pawn death.")
            (else
              (LocalVariables|SetIsKilledValid (Math|Boolean|ORBoolean (GameplayTags|IsTagQueryEmpty (Variables|UserOptions|GetTarget%20Type%20Tag)) (GameplayTags|DoesContainerMatchTagQuery (GameplayTags|GetOwnedGameplayTags (Variables|GetKilled%20AI%20Pawn) :!class /Script/GameplayTags.GameplayTagAssetInterface) (Variables|UserOptions|GetTarget%20Type%20Tag)) (GameplayTags|DoesContainerMatchTagQuery (Variables|/Script/FortniteGame.FortPawn|GetGameplayTags (Variables|GetKilled%20AI%20Pawn)) (Variables|UserOptions|GetTarget%20Type%20Tag)))))))))
    (:then_1
      (if (LocalVariables|GetIsKilledValid)
        (Default|QueuePlayerLocation (Variables|GetKilled%20AI%20Pawn))
        (GameplayEvent|BroadcastEvent (Variables|UserOptions|Events|GetOn%20Eliminated) 0 (Utilities|Struct|MakeGameplayEventwhenElimination (Pawn|GetController (Variables|GetKilled%20AI%20Pawn))))
        (Utilities|IsValid (Variables|GetKiller%20Controller)
          (:Is%20Valid
            (GameplayEvent|BroadcastEvent (Variables|UserOptions|Events|GetOn%20Elimination) 0 (Utilities|Struct|MakeGameplayEventonElimination (Variables|GetKiller%20Controller))))
          (:Is%20Not%20Valid
            (GameplayEvent|BroadcastEvent (Variables|UserOptions|Events|GetOn%20Elimination) 0 (Utilities|Struct|MakeGameplayEventonElimination 0))))))))

(fn HandlePlayerDied
  (params (FinisherOrDownerPawn :type Fort%20Player%20Pawn%20Athena%20Object%20Reference) (DyingPlayerPawn :type Fort%20Player%20Pawn%20Athena%20Object%20Reference%20%28by%20ref%29))
  (if (or (GameplayTags|IsTagQueryEmpty (Variables|UserOptions|GetTarget%20Type%20Tag)) (GameplayTags|DoesContainerMatchTagQuery (Variables|Default|GetPlayerGameplayTag) (Variables|UserOptions|GetTarget%20Type%20Tag)))
    (bind _valid_controller (CreativeUserRestrictions|ActorHasValidParticipantController (Variables|Default|GetCreative_UseRestriction_Component) DyingPlayerPawn true false (and (not (Variables|Default|GetbDisabledSpawning)) (Default|PassesSelfEliminationTest DyingPlayerPawn))))
    (if _valid_controller
      (Default|QueuePlayerLocation (Variables|GetDyingPlayerPawn))
      (GameplayEvent|BroadcastEvent (Variables|UserOptions|Events|GetOn%20Eliminated) 0 (Utilities|Struct|MakeGameplayEventwhenElimination (Pawn|GetController (Variables|GetDyingPlayerPawn))))
      (Utilities|IsValid (Variables|GetFinisherOrDownerPawn)
        (:Is%20Valid
          (GameplayEvent|BroadcastEvent (Variables|UserOptions|Events|GetOn%20Elimination) 0 (Utilities|Struct|MakeGameplayEventonElimination (Pawn|GetController (Variables|GetFinisherOrDownerPawn))))
          (if (Variables|UserOptions|GetElimination%20Penalty)
            (Default|HandleEliminationPenalties (Pawn|GetController (Variables|GetDyingPlayerPawn)) (Pawn|GetController (Variables|GetFinisherOrDownerPawn)))))
        (:Is%20Not%20Valid
          (GameplayEvent|BroadcastEvent (Variables|UserOptions|Events|GetOn%20Elimination) 0 (Utilities|Struct|MakeGameplayEventonElimination 0))
          (if (Variables|UserOptions|GetElimination%20Penalty)
            (Default|HandleEliminationPenalties (Pawn|GetController (Variables|GetDyingPlayerPawn)))))))))

(fn HandleVFX
  (params (Controller :type Controller%20Object%20Reference) (CheckOrAddTags :type Boolean))
  (bind _vfx_gameplay_tags (Variables|Default|GetVFX%20Gameplay%20Tags))
  (bind _controller (Variables|GetController))
  (if CheckOrAddTags
    (if (CreativeUserRestrictions|CanParticipantActorInteract (Variables|Default|GetCreative_UseRestriction_Component) (Variables|GetController) false true false)
      (if (Utilities|Select :Index (Variables|UserOptions|Advanced|GetElimination%20FX) :NewEnumerator0 false :NewEnumerator1 true)
        (Utilities|IsValid _controller
          (:Is%20Valid
            (Ability|GameplayEffect|AddLooseGameplayTags _controller _vfx_gameplay_tags true)))
        (else
          _)))
    (else
      _))
  (Utilities|IsValid _controller
    (:Is%20Valid
      (Ability|GameplayEffect|RemoveLooseGameplayTags _controller _vfx_gameplay_tags true))))

(fn HandleVFXAll
  (params (Check%20Or%20Add%20Tags :type Boolean))
  (bind _fortminigameprogress (Variables|Default|GetFortMinigameProgress)
    (:then
      (bind _currentminigame (Variables|/Script/FortniteGame.FortMinigameLogicComponent|GetCurrentMinigame _fortminigameprogress)
        (:then
          (bind _outplayers (Minigame|GetParticipatingPlayers_NotPure _currentminigame))
          (for _array_element _outplayers
            (CallFunction|HandleVFX (Fortnite|GetControllerfromActor _array_element) Check%20Or%20Add%20Tags)))
        (:else)))
    (:else)))

(fn IsMinigameinProgress
  (bind _fortminigameprogress (Variables|Default|GetFortMinigameProgress))
  (bind _currentminigame (Variables|/Script/FortniteGame.FortMinigameLogicComponent|GetCurrentMinigame _fortminigameprogress))
  (Utilities|FlowControl|Sequence
    (:then_0
      (Utilities|IsValid _currentminigame
        (:Is%20Valid
          (LocalVariables|SetbSuccess (Minigame|HasStarted _currentminigame :!class /Script/FortniteGame.FortMinigame)))))
    (:then_1
      (return (LocalVariables|GetbSuccess)))))

(fn IsParticipatingPlayer
  (params (Pawn :type Pawn%20Object%20Reference))
  ;[1 0.9385 0.3582] Returns whether the given pawn is a participating player in the game. Used to avoid tracking player bots as both AI and player eliminations
  (Utilities|FlowControl|Sequence
    (:then_0
      (Utilities|IsValid Pawn
        (:Is%20Valid
          (bind _asfort_player_state_athena (Utilities|Casting|CastToFortPlayerStateAthena (Variables|/Script/Engine.Pawn|GetPlayerState Pawn)))
          (return (Class|FortPlayerStateAthena|IsAnAthenaGameParticipant _asfort_player_state_athena)))))
    (:then_1
      (return false))))

(fn MigrateLegacyUserOptions
  (bind _target_type (Variables|Deprecated|GetTarget%20Type))
  (bind _returnvalue (Utilities|String|EnumtoString _target_type))
  (bind _returnvalue_2 (Utilities|String|Replace _returnvalue " "))
  (bind _target_type_2 (Class|GetClassDefaults "/CreativeCoreDevices/Device_EliminationManager_V3.Device_EliminationManager_V3_C"))
  (bind (_knob_eliminatedaffectsteam _knob_eliminatedaffectsclass) (Class|GetClassDefaults "/CreativeCoreDevices/Device_EliminationManager_V3.Device_EliminationManager_V3_C"))
  (bind _bonus_ammo_for_weapons (Class|GetClassDefaults "/CreativeCoreDevices/Device_EliminationManager_V3.Device_EliminationManager_V3_C"))
  (bind _numberofitemstodrop (Class|GetClassDefaults "/CreativeCoreDevices/Device_EliminationManager_V3.Device_EliminationManager_V3_C"))
  (bind _numberofitemstodrop_2 (Variables|Legacy|GetNumberOfItemsToDrop))
  (|CallParentFunction)
  (Utilities|FlowControl|Sequence
    (:then_0
      (Utilities|FlowControl|Sequence
        (:then_0
          ;[0 1 0.9331 0.25] Converting Invert Team and Class
          (if (!= _knob_eliminatedaffectsteam (Variables|Legacy|GetKnob_EliminatedAffectsTeam))
            (Variables|UserOptions|SetInvert%20Selected%20Team true)
            (Variables|Legacy|SetKnob_EliminatedAffectsTeam _knob_eliminatedaffectsteam)
            (LocalVariables|SetMigrated%20Options true)))
        (:then_1
          (if (!= _knob_eliminatedaffectsclass (Variables|Legacy|GetKnob_EliminatedAffectsClass))
            (Variables|UserOptions|SetInvert%20Selected%20Class true)
            (Variables|Legacy|SetKnob_EliminatedAffectsClass _knob_eliminatedaffectsclass)
            (LocalVariables|SetMigrated%20Options true)))
        (:then_2
          ;[0 1 0.9301 0.251] Splitting Number of Items to Drop to Bool/Number.
          (if (!= _numberofitemstodrop _numberofitemstodrop_2)
            (Variables|UserOptions|Basic|SetNumberOfItemsDropped_Override (!= _numberofitemstodrop_2 -1))
            (Variables|UserOptions|Basic|SetNumberOfItemsDropped (Math|Integer|Max(Integer) _numberofitemstodrop_2))
            (Variables|Legacy|SetNumberOfItemsToDrop _numberofitemstodrop)
            (LocalVariables|SetMigrated%20Options true)))
        (:then_3
          ;[0 1 0.9301 0.251] Migrating Bonus Ammo For Weapons
          (if (!= _bonus_ammo_for_weapons (Variables|Legacy|GetBonus%20Ammo%20for%20Weapons))
            (Variables|UserOptions|SetSpareWeaponAmmo_Override (not (Variables|Legacy|GetBonus%20Ammo%20for%20Weapons)))
            (Variables|UserOptions|SetSpareWeaponAmmo 0)
            (Variables|Legacy|SetBonus%20Ammo%20for%20Weapons _bonus_ammo_for_weapons)
            (LocalVariables|SetMigrated%20Options true)))
        (:then_4
          (if (Utilities|Enum|NotEqual(Enum) _target_type_2 _target_type)
            (bind _returnvalue_3 (DataRegistry|GetDataRegistryValueforUserOption (Utilities|Struct|MakeDataRegistryId "EliminationManager.TargetTag" (Utilities|String|StringToName _returnvalue_2)) (Variables|UserOptions|GetTarget%20Type%20Tag) :!class /Script/FortniteGame.FortActorOptionsComponent))
            (if _returnvalue_3
              (Variables|Deprecated|SetTarget%20Type _target_type_2)
              (LocalVariables|SetMigrated%20Options true)
              (else
                (Fortnite|FortShippingLog self (Utilities|String|Append "Device_EliminationManager_V3: Failed to find Data Registry Value for " _returnvalue_2) true)))))))
    (:then_1
      (if (LocalVariables|GetMigrated%20Options)
        (OptionData|RefreshPropertyOverrides (Variables|Default|GetToyOptionsComponent))))))

(fn OnPawnDeath
  (params (KilledPawn :type Fort%20Player%20Pawn%20Object%20Reference) (Damage :type Float%20%28double-precision%29) (InTags :type Gameplay%20Tag%20Container%20Structure) (EffectContext :type Gameplay%20Effect%20Context%20Handle%20Structure) (EventInstigator :type Controller%20Object%20Reference) (DamageCauser :type Actor%20Object%20Reference) (DBNOFinisher :type Controller%20Object%20Reference))
  (bind _asfort_player_pawn_athena (Utilities|Casting|CastToFortPlayerPawnAthena KilledPawn))
  (bind _finisherordownerpawn (Default|GetFinisherorDownerPawn _asfort_player_pawn_athena))
  (Default|HandlePlayerDied _finisherordownerpawn _asfort_player_pawn_athena)
  (return))

(fn PassesSelfEliminationTest
  (params (DyingPlayer :type Fort%20Player%20Pawn%20Athena%20Object%20Reference%20%28by%20ref%29))
  (bind _dyingplayer (Variables|GetDyingPlayer))
  (if (Variables|UserOptions|GetValid%20On%20Self%20Elimination)
    (return true)
    (else
      (Utilities|FlowControl|Sequence
        (:then_0
          ;[1 1 1] Check for non-self-elimination
          (Utilities|IsValid _dyingplayer
            (:Is%20Valid
              (bind _playerstate (Variables|/Script/Engine.Pawn|GetPlayerState _dyingplayer)
                (:then
                  (bind _asfort_player_state_athena (Utilities|Casting|CastToFortPlayerStateAthena _playerstate))
                  (bind (_returnvalue_finisherordowner _returnvalue_downer _returnvalue_bdbno _returnvalue_deathcause) (Utilities|Struct|BreakDeathInfo (Class|FortPlayerStateAthena|GetDeathInfo _asfort_player_state_athena)))
                  (return (or (Utilities|NotEqual(Object) _returnvalue_finisherordowner _asfort_player_state_athena) (and (Utilities|Enum|Equal(Enum) _returnvalue_deathcause "OutsideSafeZone") (Variables|UserOptions|GetStormEliminationsAlwaysValid)))))
                (:else)))))
        (:then_1
          (return false))))))

(fn QueuePlayerLocation
  (params (PlayerPawnToQueue :type Actor%20Object%20Reference%20%28by%20ref%29))
  (Utilities|IsValid PlayerPawnToQueue
    (:Is%20Valid
      (if (not (Variables|Default|GetbDisabledSpawning))
        (bind _bshoulddropitems (Default|ShouldRandomDropItems))
        (if _bshoulddropitems
          (Utilities|Array|Add (Variables|Default|GetQueuedSpawnLocations) (Utilities|Struct|MakeTeamPlayerLocation (Math|Conversions|ToInteger(Byte) (Team|GetActorPvPTeamIndex (Variables|GetPlayerPawnToQueue))) (Transformation|GetActorLocation (Variables|GetPlayerPawnToQueue))))
          (bind _minigame_spawner_component (Variables|Default|GetMinigame_Spawner_Component)
            (:then
              (if (not (Class|MinigameSpawnerComponent|IsSpawning _minigame_spawner_component))
                (Class|MinigameSpawnerComponent|ResetSpawner (Variables|Default|GetMinigame_Spawner_Component) false)
                (Class|MinigameSpawnerComponent|SpawnRemaining (Variables|Default|GetMinigame_Spawner_Component) true)))
            (:else)))))))

(fn ResaveActor
  (bind _minigame_spawner_component (Variables|Default|GetMinigame_Spawner_Component))
  (bind _minigame_spawner_component_2 (Variables|Default|GetMinigame_Spawner_Component))
  (Variables|/Game/Athena/Items/Traps/MinigameSpawner/Minigame_Spawner_Component.Minigame_Spawner_Component_C|SetMaxSpawns (Math|Integer|SelectInt (Variables|UserOptions|Basic|GetNumberOfItemsDropped) -1 (Variables|UserOptions|Basic|GetNumberOfItemsDropped_Override)) (Variables|Default|GetMinigame_Spawner_Component))
  (Variables|/Game/Athena/Items/Traps/MinigameSpawner/Minigame_Spawner_Component.Minigame_Spawner_Component_C|SetSpareWeaponAmmo_Override (Variables|UserOptions|GetSpareWeaponAmmo_Override) _minigame_spawner_component)
  (Variables|/Game/Athena/Items/Traps/MinigameSpawner/Minigame_Spawner_Component.Minigame_Spawner_Component_C|SetSpareWeaponAmmo (Variables|UserOptions|GetSpareWeaponAmmo) _minigame_spawner_component)
  (Variables|/Game/Athena/Items/Traps/MinigameSpawner/Minigame_Spawner_Component.Minigame_Spawner_Component_C|SetGrantExtraAmmo (Variables|Legacy|GetBonus%20Ammo%20for%20Weapons) _minigame_spawner_component)
  (Variables|/Game/Athena/Items/Traps/MinigameSpawner/Minigame_Spawner_Component.Minigame_Spawner_Component_C|SetRandomSpawnOption (Math|Conversions|ToInteger(Byte) (Variables|UserOptions|GetRandom%20Drop)) _minigame_spawner_component)
  (Variables|/Game/Athena/Items/Traps/MinigameSpawner/Minigame_Spawner_Component.Minigame_Spawner_Component_C|SetbShouldTossPickups (Utilities|Enum|Equal(Enum) (Variables|UserOptions|GetInitial%20Movement%20of%20Item) "NewEnumerator0") (Variables|Default|GetMinigame_Spawner_Component))
  (Variables|/Game/Athena/Items/Traps/MinigameSpawner/Minigame_Spawner_Component.Minigame_Spawner_Component_C|SetbShouldApplyGravityToPickups (or (Utilities|Enum|Equal(Enum) (Variables|UserOptions|GetInitial%20Movement%20of%20Item) "NewEnumerator0") (Utilities|Enum|Equal(Enum) (Variables|UserOptions|GetInitial%20Movement%20of%20Item) "NewEnumerator1")) _minigame_spawner_component_2)
  (Variables|/Game/Athena/Items/Traps/MinigameSpawner/Minigame_Spawner_Component.Minigame_Spawner_Component_C|SetIsPickupOnTouch (Variables|UserOptions|GetRun%20Over%20Pickup) _minigame_spawner_component_2)
  (Variables|/Game/Athena/Items/Traps/MinigameSpawner/Minigame_Spawner_Component.Minigame_Spawner_Component_C|SetItemSpawnScale (Math|Vector|MakeVector (Variables|UserOptions|GetItem%20Scale) (Variables|UserOptions|GetItem%20Scale) (Variables|UserOptions|GetItem%20Scale)) _minigame_spawner_component_2)
  (CallFunction|Resave))

(fn Reset
  (Utilities|FlowControl|Sequence
    (:then_0
      (Utilities|Array|Clear (Variables|Default|GetQueuedSpawnLocations))
      (bind _minigame_spawner_component (Variables|Default|GetMinigame_Spawner_Component)
        (:then
          (Class|MinigameSpawnerComponent|ResetSpawner _minigame_spawner_component true))
        (:else)))
    (:then_1
      (Default|CleanUpPlayers)
      (CallFunction|CleanUpMinigameDelegates))))

(fn SetSpawnedItemSettings
  (params (Actor :type Actor%20Object%20Reference))
  (bind (_smallicon _largeicon) (Utilities|Struct|BreakCreativeIconOption (Variables|UserOptions|GetHostile%20Badge%20Icon)))
  (bind (_smallicon_2 _largeicon_2) (Utilities|Struct|BreakCreativeIconOption (Variables|UserOptions|GetFriendly%20Badge%20Icon)))
  (bind (_smallicon_3 _largeicon_3) (Utilities|Struct|BreakCreativeIconOption (Variables|UserOptions|GetNeutral%20Badge%20Icon)))
  (Networking|SwitchHasAuthority
    (:Authority
      (Transformation|SetActorRotation Actor (Math|Rotator|MakeRotator 0.0 0.0 (.yaw (Transformation|GetActorRotation Actor))) false)
      (bind _result (ArrayMacros|GetValidIndex (Variables|Default|GetQueuedSpawnLocations))
        (:Valid%20Index
          (bind _offset (CallFunction|GetRandomSpawnOffset))
          (Transformation|SetActorLocation Actor (+ _offset (.Location_4_9F9C71554E72C837647904AD9F9B08BA (Utilities|Struct|BreakTeamPlayerLocation _result))) false false)
          (bind _asfort_pickup_creative (Utilities|Casting|CastToFort_Pickup_Creative Actor))
          (Pickup|SetCombinePickupswhenTossCompletes _asfort_pickup_creative true)
          ;[1 1 1] Update the pickup's volume to where it moved.
          ; The  one set at spawn may not be correct
          (Class|FortPickup|UpdateVolumetoNearestSpatialInterface _asfort_pickup_creative (+ _offset (.Location_4_9F9C71554E72C837647904AD9F9B08BA (Utilities|Struct|BreakTeamPlayerLocation _result))))
          (bind _asfort_pickup_creative_icon (Utilities|Casting|CastToFort_Pickup_Creative_Icon Actor))
          (bind _result_2 (ArrayMacros|GetValidIndex (Variables|Default|GetQueuedSpawnLocations))
            (:Valid%20Index
              (ProgressWidget|SetPickupIconSettings _asfort_pickup_creative_icon (Utilities|Struct|MakeCreativePickupIconInfo (Variables|UserOptions|GetFriendly%20Text) (Variables|UserOptions|GetNeutral%20Text) (Variables|UserOptions|GetHostile%20Text) (Variables|UserOptions|GetFriendly%20Badge%20Color) (Variables|UserOptions|GetNeutral%20Badge%20Color) (Variables|UserOptions|GetHostile%20Badge%20Color) _largeicon_2 _largeicon_3 _largeicon (Variables|UserOptions|GetBadgeVisibleToTeam) (Utilities|Struct|BreakTeamPlayerLocation _result_2))))))
        (:Invalid%20Index
          (Fortnite|FortShippingLog self "Elimination Manager tried to spawn an item with no queued location." true))))))

(fn SetupMinigameSpawner
  (bind _minigame_spawner_component (Variables|Default|GetMinigame_Spawner_Component))
  (bind _tospawnlist (Variables|/Script/FortniteGame.FortMinigameSpawnerComponent|GetToSpawnList _minigame_spawner_component))
  (Utilities|FlowControl|Sequence
    (:then_0
      (bind _output_get (Variables|Default|SetItemSpawnLocation (Transformation|GetWorldLocation (Variables|Default|GetCreativeEditOnlyBase)))))
    (:then_1
      (for (_ _array_index) _tospawnlist
        (Utilities|Struct|SetmembersinMinigameSpawnerSpawnParams (Utilities|Array|Get(aref) _tospawnlist _array_index) (Math|Transform|MakeTransform _output_get))))
    (:then_2
      (Class|MinigameSpawnerComponent|SetPickupSettings (Variables|Default|GetMinigame_Spawner_Component) (Creative|TeamOption|GetLegacyTeamIndex (Variables|UserOptions|GetPickup%20AllowedTeam)) (Creative|ClassOption|GetLegacyClassIndex (Variables|UserOptions|GetPickup%20Allowed%20Class)) (Math|Conversions|ToInteger(Boolean) (not (Variables|UserOptions|GetInvert%20Pickup%20Allowed%20Team))) (Math|Conversions|ToInteger(Boolean) (not (Variables|UserOptions|GetInvert%20Pickup%20Allowed%20Class))) (Variables|Default|GetItemsVisibleToAllowedOnly))
      ;[1 0.3547 0] This should be more generically added to the spawner, but for a fast fix adding it here for now. - ZakL
      (Variables|/Game/Athena/Items/Traps/MinigameSpawner/Minigame_Spawner_Component.Minigame_Spawner_Component_C|SetBasePickupClass (select (Variables|UserOptions|GetBadgeVisibleToTeam_Override) (Variables|Default|GetIcon%20Pickup%20Class) (Variables|Default|GetNormal%20Pickup%20Class)) (Variables|Default|GetMinigame_Spawner_Component)))))

(fn ShouldRandomDropItems
  (bind _output_get (LocalVariables|SetRandom%20Roll (Math|Random|RandomFloatinRange 0.001 1e+02)))
  (Math|Float|CompareFloat _output_get (Variables|UserOptions|GetDrop%20Chance)
    (:>
      _)
    (:==
      _)
    (:<
      (return true)))
  (return false))

(fn UpdateItemPreviewList
  (Networking|CanExecuteCosmeticEvents
    (:True
      (Class|CreativeItemPreviewComponent|UpdateItemListFromSpawnParams (Variables|Default|GetCreative_ItemPreview_Component) (Variables|/Script/FortniteGame.FortMinigameSpawnerComponent|GetToSpawnList (Variables|Default|GetMinigame_Spawner_Component))))))

(fn ConstructionScript
  (|CallParentFunction))
