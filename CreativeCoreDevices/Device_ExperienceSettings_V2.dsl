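; Creative device Blueprint for island Experience Settings, pushing game rules, locomotion, chat, UI and tracked stats to the minigame.
; Option changes trigger a chain of update functions on authority, with a smaller set re-run on remotes, writing class-default scalable floats and locomotion presets.
; Minigame state changes reset end callouts and register test players.
; Version checks and legacy-option migrations preserve older islands' behavior.
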(event Custom|OnMinigameStateChanged
  (params (Minigame :type Fort%20Minigame%20Object%20Reference) (MinigameState :type EFortMinigameState%20Enum :default "PreGame"))
  (bind _returnvalue (Minigame|GetMinigameSettingsMachine Minigame))
  (switch Utilities|FlowControl|Switch|SwitchonEFortMinigameState MinigameState
    (:Setup
      (CallFunction|OverrideCustomEndCallouts "" "" (Variables|UserOptions|GetGameEndCallout)) ; We're setting up a new game / round, reset the Device Vicotry Callouts \n(there are temporary overrides of minigame settings by the Game End device)
      (Utilities|IsValid Minigame
        (:Is%20Valid
          (CallFunction|UpdateMinigameStatData (Variables|/Script/FortniteGame.FortMinigame|GetVolume Minigame))
          (Class|TestPlayerDefinitionComponent|RegisterPlayspaceEvents (Variables|Default|GetTestPlayerDefinitionComponent))
          (Utilities|IsValid _returnvalue
            (:Is%20Valid
              (Minigame|OverridePostGameSettings _returnvalue (Variables|/Script/FortniteGame.FortMinigameVictoryScreenComponent|GetPostGameSettings (Variables|/Script/FortniteGame.FortMinigame|GetVictoryScreenComponent Minigame))))))))
    (:Teardown
      (Class|TestPlayerDefinitionComponent|UnregisterPlayspaceEvents (Variables|Default|GetTestPlayerDefinitionComponent)))))

(event Custom|RegisterMinigameEvents
  (bind _minigame (CallFunction|GetCurrentMinigame))
  (Utilities|IsValid _minigame
    (:Is%20Valid
      (Minigame|BindEventtoStateChangedDelegate _minigame (delegate Custom|OnMinigameStateChanged))
      (Minigame|BindEventtoPlayerAddedDelegate _minigame (EventDispatchers|CreateEvent :!event OnMinigame_PlayerAdded)))))

(event Custom|UnregisterMinigameEvents
  (bind _minigame (CallFunction|GetCurrentMinigame))
  (Utilities|IsValid _minigame
    (:Is%20Valid
      (Minigame|UnbindEventfromStateChangedDelegate _minigame (delegate Custom|OnMinigameStateChanged))
      (Minigame|UnbindEventfromPlayerAddedDelegate _minigame (EventDispatchers|CreateEvent :!event OnMinigame_PlayerAdded)))))

(event EventPostFinishSpawning
  (|CallParentFunction)
  (CallFunction|RegisterMinigameEvents))

(event EventBeginPlay
  (|CallParentFunction)
  (Networking|SwitchHasAuthority
    (:Authority
      (OptionData|AssignOnAnyPropertyChanged (Variables|Default|GetToyOptionsComponent) (EventDispatchers|CreateEvent :!event Custom|OnGameSettingsUpdated))
      (OptionData|AssignOnPropertiesLoaded (Variables|Default|GetToyOptionsComponent) (EventDispatchers|CreateEvent :!event Custom|OnGameSettingsUpdated))
      (CallFunction|FillTrackedStats))))

(event EventEndPlay
  (params (EndPlayReason :type EEndPlayReason%20Enum :default "Destroyed"))
  (|CallParentFunction EndPlayReason)
  (OptionData|UnbindEventfromOnAnyPropertyChanged (Variables|Default|GetToyOptionsComponent) (EventDispatchers|CreateEvent :!event Custom|OnGameSettingsUpdated))
  (OptionData|UnbindEventfromOnPropertiesLoaded (Variables|Default|GetToyOptionsComponent) (EventDispatchers|CreateEvent :!event Custom|OnGameSettingsUpdated))
  (CallFunction|UnregisterMinigameEvents))

(event Custom|OnGameSettingsUpdated
  (bind (_respawnspectatebehaviorscalablefloat _keepdroppeditemsbetweenrounds _environmentdamagefilter) (Class|GetClassDefaults "/Game/Athena/Playset/Minigames/Minigame_PlayerCreated.Minigame_PlayerCreated_C"))
  (bind (_textchat _bubblechat _bubblechatmaxrange _bubblechatminlifetime _bubblechatmaxlifetime _bubblechatocclusionenabled _bubblechatstyle _spawnimmunitytime _keeploadingscreen _untilminigamestate) (Class|GetClassDefaults "/CreativeCoreDevices/ExperienceSettings/ExperienceSettingsComponent_Minigame.ExperienceSettingsComponent_Minigame_C"))
  (Utilities|FlowControl|Sequence
    (:then_0
      (CallFunction|GenerateLocomotionValues)
      (CallFunction|UpdateGameMode)
      (CallFunction|UpdateClassSettings)
      (CallFunction|UpdateTimeLimit)
      (CallFunction|UpdateTotalRounds)
      (CallFunction|UpdateSpawnLocationSetting)
      (CallFunction|UpdateTeamSetting)
      (CallFunction|UpdatePostGameSpawnLocationSetting)
      (CallFunction|UpdateSpawnPadSelection)
      (CallFunction|UpdatePlayerLives)
      (CallFunction|UpdateAutoEndTeamThreshold)
      (CallFunction|UpdateTeamtoMovetowhenOutOfSpawns)
      (CallFunction|UpdateEndConditions)
      (CallFunction|UpdateGameStartCountdown)
      (CallFunction|UpdateAutoStart)
      (CallFunction|UpdateJoininProgress)
      (CallFunction|UpdateMinigameMutators)
      (Mantling|UpdateMantlingSettings)
      (CallFunction|UpdateSlowMotiononEndOfRound)
      (Default|UpdatePhysicsSettings)
      (Default|UpdateRoundVictoryAnimation)
      (Default|UpdateJamSettings)
      (Default|UpdateMovementSpeedOverrides)
      (CallFunction|UpdateLedgeLaunchingSettings)
      ; UpdateInGameExperienceSettings
      (Utilities|FlowControl|Sequence
        (:then_0
          (ScalableFloat|SetValuewithContext _textchat (Utilities|Select :Index (Variables|UserOptions|SharedSettings|GetTextChatScope) :NewEnumerator0 "ScopeOnly" :NewEnumerator1 "ScopeTeamOnly" :NewEnumerator2 "None"))
          (ScalableFloat|SetValuewithContext _bubblechat (Variables|UserOptions|SharedSettings|GetBubbleChat))
          (ScalableFloat|SetValuewithContext _bubblechatmaxrange (Variables|UserOptions|SharedSettings|GetBubbleChatMaximumRange))
          (ScalableFloat|SetValuewithContext _bubblechatminlifetime (Variables|UserOptions|SharedSettings|GetBubbleMinimumLifetime))
          (ScalableFloat|SetValuewithContext _bubblechatmaxlifetime (Variables|UserOptions|SharedSettings|GetBubbleMaximumLifetime))
          (ScalableFloat|SetValuewithContext _bubblechatocclusionenabled (Variables|UserOptions|SharedSettings|GetOccludeBubble))
          (ScalableFloat|SetValuewithContext _bubblechatstyle (Variables|UserOptions|SharedSettings|GetBubbleChatStyle))
          (ScalableFloat|SetValuewithContext _spawnimmunitytime (select (Variables|UserOptions|SharedSettings|GetbOverrideSpawnImmunityTime) (Variables|UserOptions|SharedSettings|GetSpawnImmunityTime) -1.0))
          (ScalableFloat|SetValuewithContext _keeploadingscreen (Variables|UserOptions|SharedSettings|GetKeepLoadingScreen))
          (ScalableFloat|SetValuewithContext _untilminigamestate (Variables|UserOptions|SharedSettings|GetUntilMinigameState)))
        (:then_1
          (ScalableFloat|SetValuewithContext _respawnspectatebehaviorscalablefloat (Variables|Legacy|GetRespawnSpectateBehavior))
          (ScalableFloat|SetValuewithContext _keepdroppeditemsbetweenrounds (Variables|UserOptions|GetKeepDroppedItemsBetweenRounds))
          (ScalableFloat|SetValuewithContext _environmentdamagefilter (CallFunction|GetEnvDmgFilterEnum)))
        (:then_2
          _))
      (CallFunction|UpdateVoiceChatExperienceSettings)
      (CallFunction|UpdateCoreExperienceSettings)
      (CallFunction|UpdateTestPlayers))
    (:then_1
      (CallFunction|UpdateGameSettings)
      (CallFunction|SetMinEnergytoActivateSprint)
      (Default|UpdatePatchworkSettings)
      (CallFunction|UpdateFallDamageType)
      (CallFunction|UpdateLocomotionSettings)
      (CallFunction|UpdateCommunicationOverride)
      (Default|UpdateSpectatorSettings))))

(event Minigame|EventStartMinigame_BP
  (bind _asfort_player_controller_athena (Utilities|Casting|CastToFortPlayerControllerAthena (Utilities|Array|Get(acopy) (Fortnite|GetLocalFortPlayerControllers true))))
  (Creative|ServerStartMinigame _asfort_player_controller_athena))

(event Minigame|EventAbandonMinigame_BP
  (bind _asfort_player_controller_athena (Utilities|Casting|CastToFortPlayerControllerAthena (Utilities|Array|Get(acopy) (Fortnite|GetLocalFortPlayerControllers true))))
  (Creative|ServerEndMinigame _asfort_player_controller_athena false))

(event OnAnyPropertyChanged(ToyOptionsComponent)
  (CallFunction|AdjustMinigameSettings)
  (Networking|SwitchHasAuthority
    (:Remote
      (CallFunction|GenerateLocomotionValues)
      (Default|UpdateMovementSpeedOverrides)
      (CallFunction|UpdateFallDamageType)
      (CallFunction|UpdateCoreExperienceSettings)
      (CallFunction|UpdateLocomotionSettings)
      (CallFunction|UpdateUISettings)
      (ScalableFloat|SetValuewithContext (Class|GetClassDefaults "/Game/Athena/Playset/Minigames/Minigame_PlayerCreated.Minigame_PlayerCreated_C") (CallFunction|GetEnvDmgFilterEnum)))))

(event EventApplySettingsPreset
  (params (Preset :type Fort%20Minigame%20Preset%20Structure%20%28by%20ref%29))
  (bind _teamrotation (Utilities|IfOverridesTeamRotation Preset)
    (:Overrides
      (if (== _teamrotation 0)
        (Variables|UserOptions|SetTeamRotation 0)
        (Variables|UserOptions|Overrides|SetTeamRotation_Override false)
        (else
          (Variables|UserOptions|SetTeamRotation _teamrotation)
          (Variables|UserOptions|Overrides|SetTeamRotation_Override true))))))

(event Custom|LoadAndApplyCurveTableToRegistry
  (params (CurveTable :type Soft%20Object%20Reference))
  (bind _object (Utilities|AsyncLoadAsset CurveTable)
    (:Completed
      (ScalableFloat|AddCurveTabletoDataRegistrywithContext _object "Fort.DataRegistry" true))))

(fn AdditionalUIExtensionValidityChecks
  (params (TagContainer :type Gameplay%20Tag%20Container%20Structure))
  (Utilities|FlowControl|Sequence
    (:then_0
      (if (GameplayTags|HasTag (Variables|GetTagContainer) "(TagName=\"Minigame.UIExtension.RoundStatus\")" true)
        (return (== (CallFunction|GetTeamSetting) 2))))
    (:then_1
      (return true))))

(fn AdjustMinigameSettings
  (CallFunction|UpdateMatchmakingSetings)
  (Class|TestPlayerDefinitionComponent|UpdateMaxTestPlayers (Variables|Default|GetTestPlayerDefinitionComponent) (Variables|UserOptions|GetMaxPlayers)))

(fn BlueprintCanInteract
  (params (InteractingPawn :type Fort%20Pawn%20Object%20Reference) (InteractionBeingAttempted :type EInteractionBeingAttempted%20Enum :default "FirstInteraction") (InteractionType :type TInteractionType%20Enum :default "IT_NoInteraction"))
  (bind _returnvalue (Pawn|GetController InteractingPawn))
  (Utilities|FlowControl|Sequence
    (:then_0
      (Utilities|IsValid InteractingPawn
        (:Is%20Valid
          (Utilities|IsValid _returnvalue
            (:Is%20Valid
              (bind _asfort_player_controller_athena (Utilities|Casting|CastToFortPlayerControllerAthena _returnvalue))
              (LocalVariables|SetCanInteract (.ReturnValue (Creative|CanPlayerCreateinCurrentVolume _asfort_player_controller_athena :!class /Script/FortniteGame.FortPlayerControllerAthena))))))))
    (:then_1
      (return (LocalVariables|GetCanInteract)))))

(fn BlueprintGetFailedInteractionString
  (params (InteractingPawn :type Fort%20Pawn%20Object%20Reference) (InteractionBeingAttempted :type EInteractionBeingAttempted%20Enum :default "FirstInteraction"))
  (return "NSLOCTEXT(\"[8D01EEE7B1BD51C3B901224BB424D58F]\", \"AF5FD7EC414CE3E6E764F9907D9220F3\", \"Cannot Create Here\")"))

(fn BlueprintGetInteractionString
  (params (InteractingPawn :type Fort%20Pawn%20Object%20Reference) (InteractionBeingAttempted :type EInteractionBeingAttempted%20Enum :default "FirstInteraction"))
  (return "NSLOCTEXT(\"[8D01EEE7B1BD51C3B901224BB424D58F]\", \"69ABEBB34CD837133B40D5AABDFC0D8D\", \"Modify Game Settings\")"))

(fn ConvertOldStatToNewStat
  (params (Old%20Override :type Boolean%20%28by%20ref%29) (Old%20Stat :type Enum_ExperienceSettings_TrackedStat%20Enum%20%28by%20ref%29 :default "NewEnumerator0") (New%20Override :type Boolean%20%28by%20ref%29) (New%20Stat :type Stat%20Event%20Filter%20Structure%20%28by%20ref%29))
  (Utilities|FlowControl|Sequence
    (:then_0
      (if Old%20Override
        (bind _returnvalue (DataRegistry|GetDataRegistryValueforUserOption (Utilities|Select :Index Old%20Stat :NewEnumerator0 "Creative.Stats:Eliminations" :NewEnumerator1 "Creative.Stats:EliminationAssists" :NewEnumerator2 "Creative.Stats:Eliminated" :NewEnumerator3 "Creative.Stats:CollectItems" :NewEnumerator4 "Creative.Stats:PlayerHealth" :NewEnumerator5 "Creative.Stats:AIEliminations" :NewEnumerator6 "Creative.Stats:Score" :NewEnumerator7 "Creative.Stats:Objectives" :NewEnumerator8 "Creative.Stats:Time" :NewEnumerator9 "Creative.Stats:SpawnsLeft" :NewEnumerator10 "Creative.Stats:LapTime" :NewEnumerator11 "Creative.Stats:TimeAlive" :NewEnumerator12 "Creative.Stats:DamageDealt" :NewEnumerator13 "Creative.Stats:DamageTaken" :NewEnumerator14 "Creative.Stats:RaceTime") (Variables|GetNew%20Stat) :!class /Script/FortniteGame.FortActorOptionsComponent))
        (if _returnvalue
          (Variables|SetBy-RefVar (Variables|GetNew%20Override) true)
          (Variables|SetBy-RefVar (Variables|GetOld%20Override))
          (return true))))
    (:then_1
      (return false))))

(fn FillTrackedStats
  (bind _validcolumnstats (LocalVariables|GetValidColumnStats))
  (bind _trackedscorestatfilters (Variables|Default|GetTrackedScoreStatFilters))
  (Utilities|FlowControl|Sequence
    (:then_0
      (Utilities|Array|Clear _trackedscorestatfilters)
      (Utilities|ConditionalAddTrackedStatFilter _trackedscorestatfilters (Variables|UserOptions|Overrides|GetRoundWinConditionFilter_Override) (Variables|UserOptions|GetRoundWinConditionFilter)
        (:Continue
          (Utilities|ConditionalAddTrackedStatFilter _trackedscorestatfilters (Variables|UserOptions|Overrides|GetTiebreaker1Filter_Override) (Variables|UserOptions|GetTieBreaker1Filter)
            (:Continue
              (Utilities|ConditionalAddTrackedStatFilter _trackedscorestatfilters (Variables|UserOptions|Overrides|GetTiebreaker2Filter_Override) (Variables|UserOptions|GetTieBreaker2Filter)
                (:Continue
                  (Utilities|ConditionalAddTrackedStatFilter _trackedscorestatfilters (Variables|UserOptions|Overrides|GetTiebreaker3Filter_Override) (Variables|UserOptions|GetTieBreaker3Filter)
                    (:Continue
                      (Utilities|ConditionalAddTrackedStatFilter _trackedscorestatfilters (Variables|UserOptions|Overrides|GetTiebreaker4Filter_Override) (Variables|UserOptions|GetTieBreaker4Filter)
                        (:Continue
                          ;[1 1 1] Add Game Score stat tracking, if needed.
                          (Utilities|ConditionalAddTrackedStatFilter _trackedscorestatfilters (and (Utilities|Enum|Equal(Enum) (Variables|UserOptions|GetGameWinCondition) "MostScoreWins") (not (Utilities|Array|ContainsItem _trackedscorestatfilters (LocalVariables|GetGameScoreTracking)))) (LocalVariables|GetGameScoreTracking)))))))))))))
    (:then_1
      (Utilities|ConditionalAddTrackedStatFilter _validcolumnstats (Variables|UserOptions|Overrides|GetFirstScoreboardColumnFilter_Override) (Variables|UserOptions|GetFirstScoreboardColumnFilter)
        (:Continue
          (Utilities|ConditionalAddTrackedStatFilter _validcolumnstats (Variables|UserOptions|Overrides|GetSecondScoreboardColumnFilter_Override) (Variables|UserOptions|GetSecondScoreboardColumnFilter)
            (:Continue
              (Utilities|ConditionalAddTrackedStatFilter _validcolumnstats (Variables|UserOptions|Overrides|GetThirdScoreboardColumnFilter_Override) (Variables|UserOptions|GetThirdScoreboardColumnFilter)
                (:Continue
                  (Utilities|ConditionalAddTrackedStatFilter _validcolumnstats (Variables|UserOptions|Overrides|GetFourthScoreboardColumnFilter_Override) (Variables|UserOptions|GetFourthScoreboardColumnFilter)
                    (:Continue
                      (Utilities|ConditionalAddTrackedStatFilter _validcolumnstats (Variables|UserOptions|Overrides|GetFifthScoreboardColumnFilter_Override) (Variables|UserOptions|GetFifthScoreboardColumnFilter)))))))))))
    (:then_2
      ;[1 1 1] Only override the column order if there's at least one scoreboard column option set
      (Variables|Default|SetTrackedScoreColumnOrder (select (Utilities|Array|IsNotEmpty (LocalVariables|GetValidColumnStats)) (LocalVariables|GetValidColumnStats) (Variables|Default|GetTrackedScoreStatFilters))))))

(fn FloatToMantlingHeightPreset
  (params (Min :type Float%20%28double-precision%29) (Max :type Float%20%28double-precision%29) (StepIncrement :type Float%20%28double-precision%29) (OptionValue :type Float%20%28double-precision%29))
  (return (Utilities|Enum|BytetoEnumEnum_Common_MantlingLedgeHeight (Math|Conversions|ToByte(Integer) (Math|Float|Floor (/ (- (Math|Float|Clamp(Float) OptionValue Min Max) Min) StepIncrement))))))

(fn GameModePresetUpgradePath
  (bind _gamemodeindex (Variables|Legacy|GetGameModeIndex))
  (bind _output (Utilities|GetScoreingPreset _gamemodeindex))
  (bind (_sortscoreboard_6_afc328db4172cce4e5742a9504d0c087 _scoreboardstats_7_279a843f4d5a831e0cafe5ab87e85fb6 _uiextensions_17_755f8b1c4db6c595f20783a14b316aec) (Utilities|Struct|BreakMinigameScoringPreset _output))
  (bind (_roundwincondition _roundwincondition_override _tiebreaker1 _tiebreaker1_override _tiebreaker2_override _tiebreaker3_override _tiebreaker4_override _tiebreaker2 _tiebreaker3 _tiebreaker4) (Class|GetClassDefaults "/CreativeCoreDevices/Device_ExperienceSettings_V2.Device_ExperienceSettings_V2_C"))
  (bind _returnvalue (Utilities|GetUIExtensionIndex _uiextensions_17_755f8b1c4db6c595f20783a14b316aec))
  (if (Utilities|IsValidGameModePreset _gamemodeindex)
    (Utilities|FlowControl|Sequence
      (:then_0
        (UserOptions|MigrateLegacyStatIndex _scoreboardstats_7_279a843f4d5a831e0cafe5ab87e85fb6 0 (Variables|Legacy|GetRoundWinCondition) _roundwincondition (Variables|Legacy|GetRoundWinCondition_Override) _roundwincondition_override)
        (UserOptions|MigrateLegacyStatIndex _scoreboardstats_7_279a843f4d5a831e0cafe5ab87e85fb6 1 (Variables|Legacy|GetTiebreaker1) _tiebreaker1 (Variables|Legacy|GetTiebreaker1_Override) _tiebreaker1_override)
        (UserOptions|MigrateLegacyStatIndex _scoreboardstats_7_279a843f4d5a831e0cafe5ab87e85fb6 2 (Variables|Legacy|GetTiebreaker2) _tiebreaker2 (Variables|Legacy|GetTiebreaker2_Override) _tiebreaker2_override)
        (UserOptions|MigrateLegacyStatIndex _scoreboardstats_7_279a843f4d5a831e0cafe5ab87e85fb6 3 (Variables|Legacy|GetTiebreaker3) _tiebreaker3 (Variables|Legacy|GetTiebreaker3_Override) _tiebreaker3_override)
        (UserOptions|MigrateLegacyStatIndex _scoreboardstats_7_279a843f4d5a831e0cafe5ab87e85fb6 4 (Variables|Legacy|GetTiebreaker4) _tiebreaker4 (Variables|Legacy|GetTiebreaker4_Override) _tiebreaker4_override))
      (:then_1
        (CallFunction|FillTrackedStats))
      (:then_2
        (Variables|UserOptions|Overrides|SetHudInfoType_Override (>= _returnvalue 0))
        (Variables|UserOptions|SetHUDInfoType (Utilities|Enum|BytetoEnumEnum_ExperienceSettings_HudInfoType (Math|Conversions|ToByte(Integer) _returnvalue)))))
    (else
      (Development|PrintString "MinigameSettingsMachine::GameModePresetUpgradePath - Failed"))))

(fn GenerateLocomotionValues
  (bind _locomotionpreset (Variables|UserOptions|GetLocomotionPreset))
  (bind _locomotionpresetenumtorow (Variables|Default|GetLocomotionPresetEnumToRow))
  (bind (_value _returnvalue) (Utilities|Map|Find _locomotionpresetenumtorow _locomotionpreset))
  (if (Utilities|Enum|Equal(Enum) (Variables|UserOptions|GetLocomotionPreset) "NewEnumerator0")
    _
    (elif _returnvalue
      (bind _returnvalue_2 (Utilities|GetDataTableRow (Variables|Default|GetLocomotionPresetDataTable) _value)
        (:then
          (Variables|Default|SetLocomotionValues _returnvalue_2))
        (:RowNotFound
          (Variables|Default|SetLocomotionValues (Utilities|Struct|MakeLocomotionPresetStruct (Variables|UserOptions|SharedSettings|GetEnergyMax) (Variables|UserOptions|SharedSettings|GetEnergyRechargeAmount) (Variables|UserOptions|SharedSettings|GetEnergyRechargeDelay) (Variables|UserOptions|Scoped|GetbFallDamageV2) (Variables|UserOptions|Scoped|GetFallDamageType) (Variables|UserOptions|SharedSettings|GetbApplyJumpingPenalty) (Variables|UserOptions|SharedSettings|GetbAllowMantling) (Variables|UserOptions|GetbPBWsGenerateLedges) (Variables|UserOptions|SharedSettings|GetMantlingMinimumHeight) (Variables|UserOptions|SharedSettings|GetMantlingMinimumHeightInWater) (Variables|UserOptions|SharedSettings|GetbAllowVaulting) (Variables|UserOptions|SharedSettings|GetbAllowSprinting) 0.0 (Variables|UserOptions|SharedSettings|GetSprintingEnergyCostPerSecond) (Variables|UserOptions|SharedSettings|GetMaxSprintingJumpMultiplier) (Variables|UserOptions|SharedSettings|GetMaxSprintingSpeedMultiplier) (Variables|UserOptions|SharedSettings|GetbAllowSliding) (Variables|UserOptions|SharedSettings|GetbAllowSlideKick) (Variables|UserOptions|SharedSettings|GetbAllowShoulderBashing) (Variables|UserOptions|SharedSettings|GetbGliderRedeployable) (Variables|UserOptions|SharedSettings|GetbFlyEnable) (Variables|UserOptions|SharedSettings|GetbAllowFlightSprinting) (Variables|UserOptions|SharedSettings|GetFlyingSpeedPreset) (Variables|UserOptions|Scoped|GetbEnableMutator) (Variables|UserOptions|Scoped|GetMovementSpeedTunings) (Variables|UserOptions|SharedSettings|GetbAllowBoostedJump) 0.0 (Variables|UserOptions|SharedSettings|GetbAllowRollLanding) 0.0 0.0 0.0 "NewEnumerator0" 0.0 (Variables|UserOptions|Scoped|GetbAllowWallKick) 0 0.0 0.0 (Variables|UserOptions|Scoped|GetbAllowWallScramble) (not (Variables|UserOptions|SharedSettings|GetbAllowHurdleOverJumpableObjects)) (Variables|UserOptions|GetEnergyCostOnJump) (Variables|UserOptions|GetShouldPauseEnergyUseOnFalling) false (Variables|UserOptions|GetFall%20Damage%20Capping) 0.0 false false false (select (Fortnite|IsPhysicallySimulatedIsland) false (Variables|UserOptions|GetGround%20Rush)))))))))

(fn GetCurrentMinigame
  (return (Minigame|GetMinigame)))

(fn GetDynamicUITags
  (Utilities|FlowControl|Sequence
    (:then_0
      (if (Variables|UserOptions|GetShowTopCenterScoreHUD)
        (GameplayTags|AddGameplayTag (LocalVariables|GetOutDynamicUIExtensions) (Variables|Default|GetTopCenterExtensionTag) :!class /Script/GameplayTags.BlueprintGameplayTagLibrary)))
    (:then_1
      (if (Variables|UserOptions|GetbDisplayScoreboard)
        (GameplayTags|AddGameplayTag (LocalVariables|GetOutDynamicUIExtensions) (Variables|Default|GetScoreboardContainerExtensionTag) :!class /Script/GameplayTags.BlueprintGameplayTagLibrary)))
    (:then_2
      (if (Variables|UserOptions|GetbDisplayCareerScoreboard)
        (GameplayTags|AddGameplayTag (LocalVariables|GetOutDynamicUIExtensions) (Variables|Default|GetCareerScoreboardContainerExtensionTag) :!class /Script/GameplayTags.BlueprintGameplayTagLibrary)))
    (:then_3
      (return (LocalVariables|GetOutDynamicUIExtensions)))))

(fn GetEndCallouts
  (return (select (not (Utilities|Text|TextIsEmpty (Variables|Default|GetDeviceVictoryCallout))) (Variables|Default|GetDeviceVictoryCallout) (Variables|UserOptions|GetCustomVictoryCallout)) (select (not (Utilities|Text|TextIsEmpty (Variables|Default|GetDeviceDefeatCallout))) (Variables|Default|GetDeviceDefeatCallout) (Variables|UserOptions|GetCustomDefeatCallout))))

(fn GetEnvDmgFilterEnum
  (return (Utilities|Select :Index (Variables|UserOptions|Scoped|GetEnvironmentDamagePreset) :NewEnumerator0 "All" :NewEnumerator2 "Off" :NewEnumerator1 "All" :NewEnumerator3 "PlayerBuiltOnly")))

(fn GetGameEndCallouts
  (params (bUseDeviceOveriddenCallouts :type Boolean))
  (bind (_customvictorycallout _customdefeatcallout) (CallFunction|GetEndCallouts))
  (return (select bUseDeviceOveriddenCallouts (Variables|Default|GetDeviceVictoryCallout) _customvictorycallout) (select bUseDeviceOveriddenCallouts (Variables|Default|GetDeviceDefeatCallout) _customdefeatcallout) (select bUseDeviceOveriddenCallouts (Variables|Default|GetDeviceGameEndCallout) (Variables|UserOptions|GetGameEndCallout))))

(fn GetMaxHUDTrackers
  (return (Variables|UserOptions|GetMaxTrackersOnHUD)))

(fn GetPickupSpawnLocation
  (return (Math|Vector|vector+vector (Transformation|GetActorLocation (Actor|GetOwner)) (* (Transformation|GetActorForwardVector (Actor|GetOwner)) 100) "0,0,50.000000")))

(fn GetPublishedIslandCodeDisplay
  (return (Variables|UserOptions|GetPublishedIslandCodeDisplay)))

(fn GetScoreboardStatsArray
  (for _array_element (Variables|Default|GetTrackedScoreStats)
    (bind _output (Utilities|IsValidIndex (Variables|Default|GetTrackedStats) (Math|Conversions|ToInteger(Byte) _array_element))
      (:true
        (Utilities|Array|AddUnique (LocalVariables|GetTrackedScoreboardStats) _output))))
  (return (LocalVariables|GetTrackedScoreboardStats)))

(fn GetScoreToEnd
  (return (select (Variables|UserOptions|Overrides|GetStatToEnd_Override) (Variables|UserOptions|GetStatValueToEnd) 0)))

(fn GetTrackerNumberToDisplayInHUD
  (return (Variables|UserOptions|GetMaxTrackersOnHUD)))

(fn GetUIExtensions
  (params (bIsCuratedHub :type Boolean))
  (LocalVariables|SetFinal%20UI%20Extensions (Variables|Default|GetCommonUIExtensions))
  (if (not (Variables|GetbIsCuratedHub))
    (GameplayTags|AppendGameplayTagContainers (LocalVariables|GetFinal%20UI%20Extensions) (GameplayTags|MakeGameplayTagContainerfromTag "(TagName=\"Minigame.UIExtension.OverviewGameInfoTab\")")))
  (Utilities|FlowControl|Sequence
    (:then_0
      (bind _output (Utilities|IsValidIndex (Variables|Default|GetAvailableUIExtensions) (select (Variables|UserOptions|Overrides|GetHudInfoType_Override) (Math|Conversions|ToInteger(Byte) (Variables|UserOptions|GetHUDInfoType)) -1))
        (:true
          (bind _buiextensionisvalid (CallFunction|AdditionalUIExtensionValidityChecks _output))
          (if _buiextensionisvalid
            (GameplayTags|AppendGameplayTagContainers (LocalVariables|GetFinal%20UI%20Extensions) _output)))))
    (:then_1
      (GameplayTags|AppendGameplayTagContainers (LocalVariables|GetFinal%20UI%20Extensions) (Default|GetDynamicUITags)))
    (:then_2
      (return (LocalVariables|GetFinal%20UI%20Extensions)))))

(fn HasMinigameStarted
  (bind _minigame (CallFunction|GetCurrentMinigame))
  (Utilities|IsValid _minigame
    (:Is%20Valid
      (return (Minigame|HasStarted _minigame :!class /Script/FortniteGame.FortMinigame)))
    (:Is%20Not%20Valid
      (return false))))

(fn IsJamOptionEnabled
  (return (and (CallFunction|GetJamEnabledOption (Variables|UserOptions|GetEnableJam)) (|CallParentFunction))))

(fn IsSplitscreenAllowed
  ; Proximity Chat is limited in its number of active speakers if we have splitscreen enabled, since it's not currently a setting just turn splitscreen off if prox chat is enabled
  (return (not (Variables|UserOptions|SharedSettings|GetbIsProximityChatEnabled))))

(fn HandleReleaseVersionUpgrades
  (params (LastSavedVersion :type Creative%20Version%20Id%20Structure%20%28by%20ref%29))
  (bind _binfiniteresources (Variables|Legacy|GetbInfiniteResources))
  (Utilities|FlowControl|Sequence
    (:then_0
      (CreativeVersion|CompareVersiontoTag LastSavedVersion (LocalVariables|GetVersion_EnableJam)
        (:OlderVersion
          (Variables|UserOptions|SetEnableJam "Default") ; Originally set to Off, we opted to make Default the default for both old and new islands
          (LocalVariables|SetbWasMigrated true))))
    (:then_1
      (CreativeVersion|CompareVersiontoTag (Variables|GetLastSavedVersion) (LocalVariables|GetVersion_Ch5MovementSpeedTunings)
        (:OlderVersion
          (Variables|UserOptions|Scoped|SetMovementSpeedTunings "NewEnumerator0")
          (LocalVariables|SetbWasMigrated true))))
    (:then_2
      (CreativeVersion|CompareVersiontoTag (Variables|GetLastSavedVersion) (LocalVariables|GetVersion_PBWsGenerateLedges)
        (:OlderVersion
          (Variables|UserOptions|SetbPBWsGenerateLedges false)
          (LocalVariables|SetbWasMigrated true))))
    (:then_3
      (CreativeVersion|CompareVersiontoTag (Variables|GetLastSavedVersion) (LocalVariables|GetVersion_UsingCh5DSA)
        (:OlderVersion
          (Variables|UserOptions|LegacyBehavior|SetbUseLegacyTimeOfDayManager true)
          (LocalVariables|SetbWasMigrated true))))
    (:then_4
      (CreativeVersion|CompareVersiontoTag (Variables|GetLastSavedVersion) (LocalVariables|GetVersion_BeforeInfiniteAmmoSplit)
        (:OlderVersion
          (bind _output_get (Variables|UserOptions|Scoped|SetbInfiniteCharges (Variables|UserOptions|Scoped|GetbInfiniteAmmo)))
          (Variables|UserOptions|Scoped|SetbInfiniteReserveEnergy _output_get)
          (LocalVariables|SetbWasMigrated true))))
    (:then_5
      (CreativeVersion|CompareVersiontoTag (Variables|GetLastSavedVersion) (LocalVariables|GetVersion_Ch6S1)
        (:OlderVersion
          (Variables|UserOptions|SharedSettings|SetbAllowBoostedJump false)
          (Variables|UserOptions|SharedSettings|SetbAllowRollLanding false)
          (Variables|UserOptions|SharedSettings|SetbInfiniteWorldResources _binfiniteresources)
          (Variables|UserOptions|SharedSettings|SetbInfiniteBuildingResources _binfiniteresources)
          (Variables|UserOptions|SharedSettings|SetbInfiniteGold _binfiniteresources)
          (Variables|UserOptions|Scoped|SetbInfiniteDurability true)
          (Variables|UserOptions|Scoped|SetbAllowWallKick false)
          (Variables|UserOptions|Scoped|SetbAllowWallScramble false)
          (Variables|UserOptions|Scoped|SetFallDamageType "NewEnumerator1")
          (LocalVariables|SetbWasMigrated true))))
    (:then_6
      ;[0.09498 0.15 0.07306 0.5] Text Chat scope setting rollout for existing islands: Text Chat will be automatically enabled for all existing islands that currently have Voice Chat enabled.
      ; Text Chat scope will mirror what creator set for Voice Chat.
      (CreativeVersion|CompareVersiontoTag (Variables|GetLastSavedVersion) (LocalVariables|GetVersion_TextChat)
        (:OlderVersion
          (switch Utilities|FlowControl|Switch|SwitchonEnum_ExperienceSettings_VoiceChat (Variables|UserOptions|SharedSettings|GetVoiceChat)
            (:NewEnumerator0
              (Variables|UserOptions|SharedSettings|SetTextChatScope "NewEnumerator2"))
            (:NewEnumerator1
              (Variables|UserOptions|SharedSettings|SetTextChatScope "NewEnumerator1"))
            (:NewEnumerator2
              (Variables|UserOptions|SharedSettings|SetTextChatScope "NewEnumerator0")))
          (LocalVariables|SetbWasMigrated true))))
    (:then_7
      (CreativeVersion|CompareVersiontoTag (Variables|GetLastSavedVersion) (LocalVariables|GetVersion_Feral)
        (:OlderVersion
          (Variables|UserOptions|SetLocomotionPreset "NewEnumerator0")
          (LocalVariables|SetbWasMigrated true))))
    (:then_8
      (CreativeVersion|CompareVersiontoTag (Variables|GetLastSavedVersion) (LocalVariables|GetVersion_SlowMotion)
        (:OlderVersion
          (if (Editor|InEditor)
            (Variables|UserOptions|SetbSlowMotionOnEndOfRoundV2 (Variables|Legacy|GetbSlowMotionOnEndOfRound))
            (else
              (Variables|UserOptions|SetbSlowMotionOnEndOfRoundV2 false)))
          (LocalVariables|SetbWasMigrated true))))
    (:then_9
      (CreativeVersion|CompareVersiontoTag (Variables|GetLastSavedVersion) (LocalVariables|GetVersion_3600)
        (:OlderVersion
          (Variables|UserOptions|SharedSettings|SetbAllowHurdleOverJumpableObjects true)
          ;[1 1 1] Energy Usage needs to stay at pre-36.00 values for existing islands
          (Variables|UserOptions|SetEnergyCostOnJump 1e+01)
          (Variables|UserOptions|SetShouldPauseEnergyUseOnFalling false)
          (LocalVariables|SetbWasMigrated true))))
    (:then_10
      (CreativeVersion|CompareVersiontoTag (Variables|GetLastSavedVersion) (LocalVariables|GetVersionHealthShieldVariablePeriod)
        (:OlderVersion
          (Variables|UserOptions|SharedSettings|SetHealthRechargePeriod 0.2)
          (Variables|UserOptions|SharedSettings|SetShieldRechargePeriod 0.2)
          (Variables|UserOptions|SharedSettings|SetOvershieldRechargePeriod 0.2)
          (Variables|UserOptions|SharedSettings|SetHealthRechargeApplicationAmount (Variables|Legacy|GetHealthRechargeAmount))
          (Variables|UserOptions|SharedSettings|SetShieldRechargeApplicationAmount (Variables|Legacy|GetShieldRechargeAmount))
          (Variables|UserOptions|SharedSettings|SetOvershieldRechargeApplicationAmount (Variables|Legacy|GetOvershieldRechargeRate))
          (LocalVariables|SetbWasMigrated true))))
    (:then_11
      (CreativeVersion|CompareVersiontoTag (Variables|GetLastSavedVersion) (LocalVariables|GetVersion_3700)
        (:OlderVersion
          ;[1 1 1] Migrate legacy Island Privacy for older islands
          (Variables|UserOptions|SetCreativeMatchmakingPrivacy (Variables|Legacy|GetIslandMatchmakingPrivacy))
          (LocalVariables|SetbWasMigrated true))))
    (:then_12
      (CreativeVersion|CompareVersiontoTag (Variables|GetLastSavedVersion) (LocalVariables|GetVersion_4040)
        (:OlderVersion
          ; Disable Ground Rush and Fall Damage Capping Pre 40.40
          (Variables|UserOptions|SetGround%20Rush false)
          (Variables|UserOptions|SetFall%20Damage%20Capping false)
          (if (not (Variables|Default|GetbFallDamage))
            (Variables|UserOptions|Scoped|SetbFallDamageV2 false)
            (else
              _))
          (LocalVariables|SetbWasMigrated true))))
    (:then_13
      (CreativeVersion|CompareVersiontoTag (Variables|GetLastSavedVersion) (LocalVariables|GetVersion_4110)
        (:OlderVersion
          ;[0.08131 0.15 0.07469 0.5] Enable Game Start Sound pre 41.10
          (Variables|UserOptions|SetbGameStartSoundEnabled true)
          (LocalVariables|SetbWasMigrated true))))
    (:then_14
      (if (LocalVariables|GetbWasMigrated)
        (OptionData|RefreshPropertyOverrides (Variables|Default|GetToyOptionsComponent))))))

(fn MigrateLegacyUserOptions
  (bind (_timelimit _roundtimelimit) (Class|GetClassDefaults "/CreativeCoreDevices/Device_ExperienceSettings_V2.Device_ExperienceSettings_V2_C"))
  (bind _timelimit_2 (Variables|Legacy|GetTimeLimit))
  (bind (_minimumledgeheight _minimumledgeheightwater) (Class|GetClassDefaults "/CreativeCoreDevices/Device_ExperienceSettings_V2.Device_ExperienceSettings_V2_C"))
  (bind _minimumledgeheight_2 (Variables|Legacy|GetMinimumLedgeHeight))
  (bind _minimumledgeheightwater_2 (Variables|Legacy|GetMinimumLedgeHeightWater))
  (bind _falldamagemultiplier (Class|GetClassDefaults "/CreativeCoreDevices/Device_ExperienceSettings_V2.Device_ExperienceSettings_V2_C"))
  (bind _falldamagemultiplier_2 (Variables|Legacy|GetFallDamageMultiplier))
  (bind (_maxitemslots _blimitmaxitemslots _maximumequipmentslots) (Class|GetClassDefaults "/CreativeCoreDevices/Device_ExperienceSettings_V2.Device_ExperienceSettings_V2_C"))
  (bind _returnvalue (>= (Variables|Legacy|GetMaxItemSlots) 0))
  (bind _blimitmaxitemslots_2 (Variables|Legacy|GetbLimitMaxItemSlots))
  (bind _maxitemslots_2 (Variables|Legacy|GetMaxItemSlots))
  (bind (_nameplatedisplaydistance _nameplatedisplaydistancemeters _nameplatefocusangle) (Class|GetClassDefaults "/CreativeCoreDevices/Device_ExperienceSettings_V2.Device_ExperienceSettings_V2_C"))
  (bind _nameplatedisplaydistance_2 (Variables|Legacy|GetNameplateDisplayDistance))
  (bind _returnvalue_2 (> _nameplatedisplaydistance_2 0))
  (bind _nameplatefocusangle_2 (Variables|Legacy|GetNameplateFocusAngle))
  (bind (_spawnimmunitytime_legacy _spawnimmunitytime) (Class|GetClassDefaults "/CreativeCoreDevices/Device_ExperienceSettings_V2.Device_ExperienceSettings_V2_C"))
  (bind _spawnimmunitytime_legacy_2 (Variables|Legacy|GetSpawnImmunityTime_Legacy))
  (bind _ballowspectateotherteams (Class|GetClassDefaults "/CreativeCoreDevices/Device_ExperienceSettings_V2.Device_ExperienceSettings_V2_C"))
  (bind _ballowspectateotherteams_2 (Variables|Legacy|GetbAllowSpectateOtherTeams))
  (bind _bshowislandcodewatermark (Class|GetClassDefaults "/CreativeCoreDevices/Device_ExperienceSettings_V2.Device_ExperienceSettings_V2_C"))
  (bind _bshowislandcodewatermark_2 (Variables|Legacy|GetbShowIslandCodeWatermark))
  (bind _sprintingjumpmultiplier (Class|GetClassDefaults "/CreativeCoreDevices/Device_ExperienceSettings_V2.Device_ExperienceSettings_V2_C"))
  (bind _sprintingspeedmultiplier (Class|GetClassDefaults "/CreativeCoreDevices/Device_ExperienceSettings_V2.Device_ExperienceSettings_V2_C"))
  (bind (_bapplyjumppenalty _bgliderredeploy _bflyenabled _ballowflightsprint _flyspeedpreset) (Class|GetClassDefaults "/CreativeCoreDevices/Device_ExperienceSettings_V2.Device_ExperienceSettings_V2_C"))
  (bind (_numcollectitemsforendcondition _collectitemstoend _collectitemcount) (Class|GetClassDefaults "/CreativeCoreDevices/Device_ExperienceSettings_V2.Device_ExperienceSettings_V2_C"))
  (bind _numcollectitemsforendcondition_2 (Variables|Legacy|GetNumCollectItemsForEndCondition))
  (bind (_allowtacticalovershield _allowclambering _allowsprint _canslide _allowdoorbash _allowhealthshieldregenhealth _allowhealthshieldregenshield) (Class|GetClassDefaults "/CreativeCoreDevices/Device_ExperienceSettings_V2.Device_ExperienceSettings_V2_C"))
  (bind _resourcerateoverrideindex (Class|GetClassDefaults "/CreativeCoreDevices/Device_ExperienceSettings_V2.Device_ExperienceSettings_V2_C"))
  (bind _resourcerateoverrideindex_2 (Variables|Legacy|GetResourceRateOverrideIndex))
  (bind (_weapondestruction _weapondestruction_legacy _weapondestructionpercentage) (Class|GetClassDefaults "/CreativeCoreDevices/Device_ExperienceSettings_V2.Device_ExperienceSettings_V2_C"))
  (bind _weapondestruction_legacy_2 (Variables|Legacy|GetWeaponDestruction_Legacy))
  (bind _returnvalue_3 (< _weapondestruction_legacy_2 0))
  (bind _weapondestruction_2 (Variables|Legacy|GetWeaponDestruction))
  (bind _returnvalue_4 (< _weapondestruction_2 0))
  (bind _environmentdamage_legacy (Class|GetClassDefaults "/CreativeCoreDevices/Device_ExperienceSettings_V2.Device_ExperienceSettings_V2_C"))
  (bind _environmentdamage_legacy_2 (Variables|Legacy|GetEnvironmentDamage_Legacy))
  (bind _pickaxedestruction_legacy (Class|GetClassDefaults "/CreativeCoreDevices/Device_ExperienceSettings_V2.Device_ExperienceSettings_V2_C"))
  (bind _pickaxedestruction_legacy_2 (Variables|Legacy|GetPickaxeDestruction_Legacy))
  (bind _pvppickaxedamage_legacy (Class|GetClassDefaults "/CreativeCoreDevices/Device_ExperienceSettings_V2.Device_ExperienceSettings_V2_C"))
  (bind _pvppickaxedamage_legacy_2 (Variables|Legacy|GetPVPPickaxeDamage_Legacy))
  (|CallParentFunction)
  (Utilities|FlowControl|Sequence
    (:then_0
      (if (!= (Variables|Legacy|GetGameModeIndex) -1)
        (CallFunction|GameModePresetUpgradePath)
        (Variables|Legacy|SetGameModeIndex -1)
        (LocalVariables|SetbWasMigrated true)))
    (:then_1
      ; MigrateTimeLimit
      (Utilities|FlowControl|Sequence
        (:then_0
          (if (!= _timelimit _timelimit_2)
            (bind _output_get (Variables|UserOptions|Overrides|SetRoundTimeLimit_Override (> _timelimit_2 0)))
            (Variables|UserOptions|SetRoundTimeLimit (select _output_get (/ _timelimit_2 6e+01) _roundtimelimit))
            (Variables|Legacy|SetTimeLimit _timelimit)
            (Variables|SetbWasMigrated true)))
        (:then_1
          _))
      ; MigrateSharedSettings
      (Utilities|FlowControl|Sequence
        (:then_0
          (bind _bmigrated (UserOptions|MigrateFloattoBool (Variables|Legacy|GetAllowTacticalOvershield) _allowtacticalovershield (Variables|UserOptions|SharedSettings|GetbAllowOvershield)))
          (bind _bmigrated_2 (UserOptions|MigrateFloattoBool (Variables|Legacy|GetAllowClambering) _allowclambering (Variables|UserOptions|SharedSettings|GetbAllowMantling)))
          (bind _bmigrated_3 (UserOptions|MigrateFloattoBool (Variables|Legacy|GetAllowSprint) _allowsprint (Variables|UserOptions|SharedSettings|GetbAllowSprinting)))
          (bind _bmigrated_4 (UserOptions|MigrateFloattoBool (Variables|Legacy|GetCanSlide) _canslide (Variables|UserOptions|SharedSettings|GetbAllowSliding)))
          (bind _bmigrated_5 (UserOptions|MigrateFloattoBool (Variables|Legacy|GetAllowDoorBash) _allowdoorbash (Variables|UserOptions|SharedSettings|GetbAllowShoulderBashing)))
          (bind _bmigrated_6 (UserOptions|MigrateFloattoBool (Variables|Legacy|GetAllowHealthShieldRegenHealth) _allowhealthshieldregenhealth (Variables|UserOptions|SharedSettings|GetAllowHealthRecharge)))
          (bind _bmigrated_7 (UserOptions|MigrateFloattoBool (Variables|Legacy|GetAllowHealthShieldRegenShield) _allowhealthshieldregenshield (Variables|UserOptions|SharedSettings|GetAllowShieldRecharge)))
          (Variables|SetbWasMigrated (Math|Boolean|ORBoolean _bmigrated_7 _bmigrated_6 _bmigrated_5 _bmigrated_4 _bmigrated_3 _bmigrated_2 _bmigrated (Variables|GetbWasMigrated))))
        (:then_1
          _))
      ; MigrateMantlingOptions
      (Utilities|FlowControl|Sequence
        (:then_0
          (if (!= _minimumledgeheight_2 _minimumledgeheight)
            (Variables|UserOptions|SharedSettings|SetMantlingMinimumHeight (Mantling|FloattoMantlingHeightPreset (Variables|Constants|Mantling|GetMantlingHeightMin) (Variables|Constants|Mantling|GetMantlingHeightMax) (Variables|Constants|Mantling|GetMantlingHeightIncrement) _minimumledgeheight_2))
            (Variables|Legacy|SetMinimumLedgeHeight _minimumledgeheight)
            (Variables|SetbWasMigrated true)))
        (:then_1
          (if (!= _minimumledgeheightwater _minimumledgeheightwater_2)
            (Variables|UserOptions|SharedSettings|SetMantlingMinimumHeightInWater (Mantling|FloattoMantlingHeightPreset (Variables|Constants|Mantling|GetMantlingHeightWaterMin) (Variables|Constants|Mantling|GetMantlingHeightWaterMax) (Variables|Constants|Mantling|GetMantlingHeightWaterIncrement) _minimumledgeheightwater_2))
            (Variables|Legacy|SetMinimumLedgeHeightWater _minimumledgeheightwater)
            (Variables|SetbWasMigrated true)))
        (:then_2
          _))
      ; MigrateCollectItemsToEnd
      (Utilities|FlowControl|Sequence
        (:then_0
          (if (!= _numcollectitemsforendcondition _numcollectitemsforendcondition_2)
            (bind _output_get_2 (Variables|UserOptions|Overrides|SetCollectItemsToEnd_Override (!= _numcollectitemsforendcondition_2 0)))
            (bind _output_get_3 (Variables|UserOptions|SetCollectItemsToEnd (select _output_get_2 (select (> _numcollectitemsforendcondition_2 0) "NewEnumerator1" "NewEnumerator0") _collectitemstoend)))
            (Variables|UserOptions|SetCollectItemCount (Utilities|Select :Index _output_get_3 :NewEnumerator2 _collectitemcount :NewEnumerator0 _collectitemcount :NewEnumerator1 _numcollectitemsforendcondition_2))
            (Variables|Legacy|SetNumCollectItemsForEndCondition _numcollectitemsforendcondition)
            (Variables|SetbWasMigrated true)))
        (:then_1
          _))
      ; MigrateStartingHealth
      (Utilities|FlowControl|Sequence
        (:then_0
          (if (<= (Variables|UserOptions|SharedSettings|GetStartingHealthPercentage) 0)
            (Variables|UserOptions|Scoped|SetbInvincibility true)
            (Variables|UserOptions|SharedSettings|SetStartingHealthPercentage (Class|GetClassDefaults "/CreativeCoreDevices/Device_ExperienceSettings_V2.Device_ExperienceSettings_V2_C"))
            (Variables|SetbWasMigrated true)))
        (:then_1
          _))
      ; MigrateResourceRateOverrideIndex
      (Utilities|FlowControl|Sequence
        (:then_0
          (if (!= _resourcerateoverrideindex _resourcerateoverrideindex_2)
            (Variables|UserOptions|Scoped|SetHarvestStyle (Utilities|Enum|BytetoEnumEHarvestStyleOverride (Math|Conversions|ToByte(Integer) (+ _resourcerateoverrideindex_2 1))))
            (Variables|Legacy|SetResourceRateOverrideIndex _resourcerateoverrideindex)
            (Variables|SetbWasMigrated true)))
        (:then_1
          _))
      ; MigrateSprintSpeedMultipliers
      (Utilities|FlowControl|Sequence
        (:then_0
          (if (not (Math|Float|NearlyEqual(Float) (Variables|Legacy|GetSprintingSpeedMultiplier) _sprintingspeedmultiplier 0.001))
            (Variables|UserOptions|SharedSettings|SetMaxSprintingSpeedMultiplier (+ (* (- (Variables|Legacy|GetSprintingSpeedMultiplier) 1.0) 1e+01) 1.0))
            (Variables|Legacy|SetSprintingSpeedMultiplier _sprintingspeedmultiplier)
            (Variables|SetbWasMigrated true)))
        (:then_1
          (if (not (Math|Float|NearlyEqual(Float) (Variables|Legacy|GetSprintingJumpMultiplier) _sprintingjumpmultiplier 0.001))
            (Variables|UserOptions|SharedSettings|SetMaxSprintingJumpMultiplier (+ (* (- (Variables|Legacy|GetSprintingJumpMultiplier) 1.0) 1e+01) 1.0))
            (Variables|Legacy|SetSprintingJumpMultiplier _sprintingjumpmultiplier)
            (Variables|SetbWasMigrated true)))
        (:then_2
          _)))
    (:then_2
      ; MigrateWeaponDestruction
      (Utilities|FlowControl|Sequence
        (:then_0
          (if (!= _weapondestruction_legacy _weapondestruction_legacy_2)
            (Variables|UserOptions|Scoped|SetWeaponDestructionPreset (select _returnvalue_3 "NewEnumerator2" "NewEnumerator1"))
            (Variables|UserOptions|Scoped|SetWeaponDestructionPercentage (select _returnvalue_3 _weapondestructionpercentage (* _weapondestruction_legacy_2 1e+02)))
            (Variables|Legacy|SetWeaponDestruction_Legacy _weapondestruction_legacy)
            (Variables|SetbWasMigrated true)))
        (:then_1
          (if (!= _weapondestruction _weapondestruction_2)
            (Variables|UserOptions|Scoped|SetWeaponDestructionPreset (select _returnvalue_4 "NewEnumerator2" "NewEnumerator1"))
            (Variables|UserOptions|Scoped|SetWeaponDestructionPercentage (select _returnvalue_4 _weapondestructionpercentage _weapondestruction_2))
            (Variables|Legacy|SetWeaponDestruction _weapondestruction)
            (Variables|SetbWasMigrated true)))
        (:then_2
          _))
      ; MigrateEnvironmentDamage
      (Utilities|FlowControl|Sequence
        (:then_0
          (if (!= _environmentdamage_legacy _environmentdamage_legacy_2)
            (Variables|UserOptions|Scoped|SetEnvironmentDamagePreset (select (< _environmentdamage_legacy_2 0) "NewEnumerator3" (select (> _environmentdamage_legacy_2 0) "NewEnumerator1" "NewEnumerator2")))
            (Variables|Legacy|SetEnvironmentDamage_Legacy _environmentdamage_legacy)
            (Variables|SetbWasMigrated true)))
        (:then_1
          _))
      ; MigratePickaxeDestruction
      (Utilities|FlowControl|Sequence
        (:then_0
          (if (!= _pickaxedestruction_legacy _pickaxedestruction_legacy_2)
            (Variables|UserOptions|Scoped|SetPickaxeDestructionPreset (select (< _pickaxedestruction_legacy_2 0) "NewEnumerator3" (select (> _pickaxedestruction_legacy_2 0) "NewEnumerator2" "NewEnumerator1")))
            (Variables|Legacy|SetPickaxeDestruction_Legacy _pickaxedestruction_legacy)
            (Variables|SetbWasMigrated true)))
        (:then_1
          _))
      ; MigratePVPPickaxeDamage
      (Utilities|FlowControl|Sequence
        (:then_0
          (if (!= _pvppickaxedamage_legacy _pvppickaxedamage_legacy_2)
            (Variables|Legacy|SetPVPPickaxeDamagePreset (select (> _pvppickaxedamage_legacy_2 0) "NewEnumerator2" "NewEnumerator1"))
            (Variables|Legacy|SetPVPPickaxeDamage_Legacy _pvppickaxedamage_legacy)
            (Variables|SetbWasMigrated true)))
        (:then_1
          (if (Utilities|Enum|NotEqual(Enum) (Class|GetClassDefaults "/CreativeCoreDevices/Device_ExperienceSettings_V2.Device_ExperienceSettings_V2_C") (Variables|Legacy|GetPVPPickaxeDamagePreset))
            (Variables|UserOptions|Scoped|SetPickaxeDamagePreset (Utilities|Select :Index (Variables|Legacy|GetPVPPickaxeDamagePreset) :NewEnumerator0 "NewEnumerator0" :NewEnumerator1 "NewEnumerator2" :NewEnumerator2 "NewEnumerator0"))
            (Variables|SetbWasMigrated true)))
        (:then_2
          _))
      ; MigrateFallDamage
      (Utilities|FlowControl|Sequence
        (:then_0
          (if (!= _falldamagemultiplier _falldamagemultiplier_2)
            (Variables|UserOptions|Scoped|SetbFallDamageV2 (!= _falldamagemultiplier_2 0))
            (Variables|Legacy|SetFallDamageMultiplier _falldamagemultiplier)
            (Variables|SetbWasMigrated true)))
        (:then_1
          _))
      ; MigrateMaxItemSlots
      (Utilities|FlowControl|Sequence
        (:then_0
          (if (or (!= _maxitemslots _maxitemslots_2) (!= _blimitmaxitemslots _blimitmaxitemslots_2))
            (Variables|UserOptions|Scoped|SetMaximumEquipmentSlots (select _returnvalue (Math|Float|Truncate _maxitemslots_2) _maximumequipmentslots))
            (Variables|UserOptions|Overrides|SetMaximumEquipmentSlots_Override (or _returnvalue _blimitmaxitemslots_2))
            (Variables|Legacy|SetbLimitMaxItemSlots _blimitmaxitemslots)
            (Variables|Legacy|SetMaxItemSlots _maxitemslots)
            (Variables|SetbWasMigrated true)))
        (:then_1
          _)))
    (:then_3
      ; MigrateNameplateSettings
      (Utilities|FlowControl|Sequence
        (:then_0
          (if (!= _nameplatedisplaydistance _nameplatedisplaydistance_2)
            (Variables|UserOptions|Scoped|SetbLimitNamePlateMaxDistance _returnvalue_2)
            (Variables|UserOptions|Scoped|SetNameplateDisplayDistanceMeters (select _returnvalue_2 (/ _nameplatedisplaydistance_2 1e+02) _nameplatedisplaydistancemeters))
            (Variables|Legacy|SetNameplateDisplayDistance _nameplatedisplaydistance)
            (Variables|SetbWasMigrated true)))
        (:then_1
          (if (!= _nameplatefocusangle _nameplatefocusangle_2)
            (Variables|UserOptions|Scoped|SetNameplateFocusAngleDegrees (Math|Conversions|ToFloat(Integer) (Math|Float|Round (Math|Trig|RadiansToDegrees _nameplatefocusangle_2))))
            (Variables|Legacy|SetNameplateFocusAngle _nameplatefocusangle)
            (Variables|SetbWasMigrated true)))
        (:then_2
          (if (< (Variables|UserOptions|Scoped|GetNameplateFocusTime) 0)
            (Variables|UserOptions|Scoped|SetNameplateFocusTime 0.0)
            (Variables|SetbWasMigrated true)))
        (:then_3
          _))
      ; MigrateSpawnImmunityTime
      (Utilities|FlowControl|Sequence
        (:then_0
          (if (!= _spawnimmunitytime_legacy _spawnimmunitytime_legacy_2)
            (bind _output_get_4 (Variables|UserOptions|SharedSettings|SetbOverrideSpawnImmunityTime (>= _spawnimmunitytime_legacy_2 0)))
            (Variables|UserOptions|SharedSettings|SetSpawnImmunityTime (select _output_get_4 _spawnimmunitytime_legacy_2 _spawnimmunitytime))
            (Variables|Legacy|SetSpawnImmunityTime_Legacy _spawnimmunitytime_legacy)
            (Variables|SetbWasMigrated true)))
        (:then_1
          _))
      ; MigrateShowIslandCodeWatermark
      (Utilities|FlowControl|Sequence
        (:then_0
          (if (!= _bshowislandcodewatermark _bshowislandcodewatermark_2)
            (Variables|UserOptions|SetPublishedIslandCodeDisplay (select _bshowislandcodewatermark_2 "ShowWithWatermark" "ShowWithoutWatermark"))
            (Variables|Legacy|SetbShowIslandCodeWatermark _bshowislandcodewatermark)
            (Variables|SetbWasMigrated true)))
        (:then_1
          _))
      ; MigrateAllowSpectatingOtherTeams
      (Utilities|FlowControl|Sequence
        (:then_0
          (if (!= _ballowspectateotherteams _ballowspectateotherteams_2)
            (Variables|UserOptions|Scoped|SetAllowSpectateOtherTeams (select _ballowspectateotherteams_2 "Allow" "Disallow"))
            (Variables|Legacy|SetbAllowSpectateOtherTeams _ballowspectateotherteams)
            (Variables|SetbWasMigrated true)))
        (:then_1
          _))
      ; ConvertScoreToEndOverride
      (if (Variables|Legacy|GetScoreToEnd_Override)
        (Variables|UserOptions|Overrides|SetStatToEnd_Override true)
        (Variables|Legacy|SetScoreToEnd_Override false)
        (Variables|UserOptions|SetStatValueToEnd (Variables|Legacy|GetScoreToEnd))
        (Variables|Legacy|SetScoreToEnd (Class|GetClassDefaults "/CreativeCoreDevices/Device_ExperienceSettings_V2.Device_ExperienceSettings_V2_C"))
        (Variables|SetbWasMigrated true)))
    (:then_4
      ; MigrateLocomotionSettings
      (Utilities|FlowControl|Sequence
        (:then_0
          (bind _bmigrated_8 (Creative|Migration|MigrateOptionTypeBooleantoBoolean (Variables|Legacy|GetbApplyJumpPenalty) _bapplyjumppenalty (Variables|UserOptions|SharedSettings|GetbApplyJumpingPenalty)))
          (bind _bmigrated_9 (Creative|Migration|MigrateOptionTypeBooleantoBoolean (Variables|Legacy|GetbGliderRedeploy) _bgliderredeploy (Variables|UserOptions|SharedSettings|GetbGliderRedeployable)))
          (bind _bmigrated_10 (Creative|Migration|MigrateOptionTypeBooleantoBoolean (Variables|Legacy|GetbFlyEnabled) _bflyenabled (Variables|UserOptions|SharedSettings|GetbFlyEnable)))
          (bind _bmigrated_11 (Creative|Migration|MigrateOptionTypeBooleantoBoolean (Variables|Legacy|GetbAllowFlightSprint) _ballowflightsprint (Variables|UserOptions|SharedSettings|GetbAllowFlightSprinting)))
          (Variables|SetbWasMigrated (Math|Boolean|ORBoolean _bmigrated_11 _bmigrated_10 _bmigrated_9 _bmigrated_8)))
        (:then_1
          (if (!= (Math|Conversions|ToInteger(Byte) _flyspeedpreset) (Math|Conversions|ToInteger(Byte) (Variables|Legacy|GetFlySpeedPreset)))
            (Variables|UserOptions|SharedSettings|SetFlyingSpeedPreset (Variables|Legacy|GetFlySpeedPreset))
            (Variables|Legacy|SetFlySpeedPreset _flyspeedpreset)
            (Variables|SetbWasMigrated true)))
        (:then_2
          _))
      ; MigrateRespawnSpectateBehaviorToAllowSpectateTeammatesx
      (Utilities|FlowControl|Sequence
        (:then_0
          (if (Utilities|Enum|NotEqual(Enum) (Class|GetClassDefaults "/CreativeCoreDevices/Device_ExperienceSettings_V2.Device_ExperienceSettings_V2_C") (Variables|Legacy|GetRespawnSpectateBehavior))
            (if (Utilities|Enum|Equal(Enum) (Variables|Legacy|GetRespawnSpectateBehavior) "BlackScreen")
              (Variables|UserOptions|Scoped|SetAllowSpectateOtherTeams "None")
              (Variables|Legacy|SetRespawnSpectateBehavior (Class|GetClassDefaults "/CreativeCoreDevices/Device_ExperienceSettings_V2.Device_ExperienceSettings_V2_C"))
              (Variables|SetbWasMigrated true))))
        (:then_1
          _)))
    (:then_5
      (if (LocalVariables|GetbWasMigrated)
        (OptionData|RefreshPropertyOverrides (Variables|Default|GetToyOptionsComponent))))))

(fn MantlingHeightPresetToFloat
  (params (Min :type Float%20%28double-precision%29) (Max :type Float%20%28double-precision%29) (StepIncrement :type Float%20%28double-precision%29) (OptionValue :type Byte))
  (return (Math|Float|Min(Float) (+ Min (* StepIncrement (Math|Conversions|ToFloat(Byte) OptionValue))) Max)))

(fn MigrateLegacyStatIndex
  (params (TrackedStatArray :type Array%20of%20Fort%20Minigame%20Stat%20Filter%20Class%20References) (TrackedStatIndex :type Integer) (NewVariable :type Enum_ExperienceSettings_TrackedStat%20Enum%20%28by%20ref%29 :default "NewEnumerator0") (NewVariableDefault :type Enum_ExperienceSettings_TrackedStat%20Enum :default "NewEnumerator0") (NewVariableOverride :type Boolean%20%28by%20ref%29) (NewVariableOverrideDefault :type Boolean))
  (bind _output (Utilities|Array|Get(acopy) TrackedStatArray TrackedStatIndex))
  (bind (_bisvalid _returnvalue) (Utilities|GetTrackedStatIndex _output NewVariableDefault))
  (if (Utilities|Array|IsValidIndex TrackedStatArray TrackedStatIndex)
    (Variables|SetBy-RefVar NewVariableOverride _bisvalid)
    (Variables|SetBy-RefVar NewVariable _returnvalue)
    (else
      (Variables|SetBy-RefVar NewVariableOverride NewVariableOverrideDefault)
      (Variables|SetBy-RefVar NewVariable NewVariableDefault))))

(fn MigrateFloatToBool
  (params (OldOptionVariable :type Float%20%28double-precision%29%20%28by%20ref%29) (OldOptionVariableDefault :type Float%20%28double-precision%29) (NewOptionVariable :type Boolean%20%28by%20ref%29))
  (Utilities|FlowControl|Sequence
    (:then_0
      (if (!= OldOptionVariable OldOptionVariableDefault)
        (Variables|SetBy-RefVar NewOptionVariable (!= OldOptionVariable 0))
        (Variables|SetBy-RefVar OldOptionVariable OldOptionVariableDefault)
        (LocalVariables|SetbWasMigrated true)))
    (:then_1
      (return (LocalVariables|GetbWasMigrated)))))

(fn OnMinigame_PlayerAdded
  (params (UniqueNetId :type Unique%20Net%20Id%20Structure) (bIsLocalPlayer :type Boolean))
  ;[1 1 1] Prep the Use of Test player as soon as a player is added to the minigame
  (Networking|SwitchHasAuthority
    (:Authority
      (Utilities|FlowControl|DoOnce false
        (:Completed
          (Utilities|Time|SetTimerforNextTickbyEvent (EventDispatchers|CreateEvent :!event UpdateTestPlayers)) ; we have to wait one frame for the playspace component to be properly added
        )))))

(fn OverrideCustomEndCallouts
  (params (DeviceVictoryCallout :type Text) (DeviceDefeatCallout :type Text) (GameEndCallout :type EMinigameGameEndCallout%20Enum :default "WinLose"))
  (Networking|SwitchHasAuthority
    (:Authority
      (Variables|Default|SetDeviceVictoryCallout DeviceVictoryCallout)
      (Variables|Default|SetDeviceDefeatCallout DeviceDefeatCallout)
      (Variables|Default|SetDeviceGameEndCallout GameEndCallout)
      (CallFunction|UpdateGameSettings))))

(fn OverrideGameEndCallouts
  (params (VictoryCallout :type Text) (DefeatCallout :type Text) (GameEndCallout :type EMinigameGameEndCallout%20Enum :default "WinLose"))
  (CallFunction|OverrideCustomEndCallouts VictoryCallout DefeatCallout GameEndCallout)
  (return true))

(fn SetMinEnergyToActivateSprint
  ;[0 1 0.8074 0.25] If Energy Max is more than 10, clamp Min Energy to activate sprint to it's default, otherwise, match it to current Max Energy
  (ScalableFloat|SetValuewithContext (Variables|Default|GetMinActivateEnergyRow) (Math|Float|Clamp(Float) (Variables|UserOptions|SharedSettings|GetEnergyMax) 1.0 1e+01)))

(fn ShouldShowCumulativeScoreInHUD
  (return (Variables|UserOptions|GetbShowCumulativeStatValueOnHUD)))

(fn StatFilterStructToStatFilterObjectArray
  (params (InSubclassArray :type Array%20of%20Stat%20Event%20Filter%20Structures))
  (for _array_element InSubclassArray
    (Utilities|Array|Add (LocalVariables|GetLoadedStats) (Class|FortCreativeStatManagerBlueprintLibrary|GetStatFilterfromStatFilterStruct self _array_element)))
  (return (LocalVariables|GetLoadedStats)))

(fn StatFilterSubclassToLoadedArray
  (params (InSubclassArray :type Array%20of%20Fort%20Minigame%20Stat%20Filter%20Class%20References))
  (for _array_element InSubclassArray
    (Utilities|Array|Add (LocalVariables|GetLoadedStats) (Class|FortMinigameStatsComponent|GetLoadedStatfromClass (Variables|Default|GetCachedMinigameStatsComponent) _array_element)))
  (return (LocalVariables|GetLoadedStats)))

(fn TryAddEndCondition_CollectItems
  (params (EndConditions :type Array%20of%20Fort%20Minigame%20Stat%20Query%20Structures))
  (Utilities|FlowControl|Sequence
    (:then_0
      (if (Variables|UserOptions|Overrides|GetCollectItemsToEnd_Override)
        (Utilities|Array|Add EndConditions (Utilities|Struct|MakeFortMinigameStatQuery (Class|FortMinigameStatsComponent|GetLoadedStatfromClass (Variables|Default|GetCachedMinigameStatsComponent) "/Game/Athena/Playset/Minigames/StatFilters/MinigameStat_Challenge_All.MinigameStat_Challenge_All_C") "Team" true "GreaterOrEqual" (Utilities|Select :Index (Variables|UserOptions|GetCollectItemsToEnd) :NewEnumerator2 0 :NewEnumerator0 -1 :NewEnumerator1 (Math|Integer|Max(Integer) (Variables|UserOptions|GetCollectItemCount) 1)) (Utilities|Enum|Equal(Enum) (Variables|UserOptions|GetCollectItemsToEnd) "NewEnumerator1")))
        (LocalVariables|SetbAdded true)))
    (:then_1
      (return (LocalVariables|GetbAdded)))))

(fn TryAddEndCondition_DestroyObjects
  (params (EndConditions :type Array%20of%20Fort%20Minigame%20Stat%20Query%20Structures))
  (Utilities|FlowControl|Sequence
    (:then_0
      (if (Variables|UserOptions|Overrides|GetObjectivesToEnd_Override)
        (Utilities|Array|Add EndConditions (Utilities|Struct|MakeFortMinigameStatQuery (Class|FortMinigameStatsComponent|GetLoadedStatfromClass (Variables|Default|GetCachedMinigameStatsComponent) "/Game/Athena/Playset/Minigames/StatFilters/MinigameStat_Destructible_All.MinigameStat_Destructible_All_C") "Team" true "GreaterOrEqual" (Variables|UserOptions|GetObjectivesToEnd) (> (Variables|UserOptions|GetObjectivesToEnd) 0)))
        (LocalVariables|SetbAdded true)))
    (:then_1
      (return (LocalVariables|GetbAdded)))))

(fn UpdateRoundVictoryAnimation
  (Networking|SwitchHasAuthority
    (:Authority
      (bind _output_get (LocalVariables|SetMinigame (CallFunction|GetCurrentMinigame)))
      (Utilities|IsValid _output_get
        (:Is%20Valid
          (Variables|/Script/FortniteGame.FortMinigameVictoryScreenComponent|SetRoundVictoryAnimation (Variables|UserOptions|GetVictoryAnimation) (Variables|/Script/FortniteGame.FortMinigame|GetVictoryScreenComponent _output_get)))))))

(fn UpdateAutoEndTeamThreshold
  (Networking|SwitchHasAuthority
    (:Authority
      (bind _output_get (LocalVariables|SetMinigame (CallFunction|GetCurrentMinigame)))
      (Utilities|IsValid _output_get
        (:Is%20Valid
          (Variables|/Script/FortniteGame.FortMinigame|SetAutoEndTeamThreshold (Math|Conversions|ToInteger(Boolean) (Variables|UserOptions|GetbLastStandingEndsGame)) _output_get))))))

(fn UpdateAutoStart
  ;[1 0 0.03385] TODO: Rework how Minigame is handling AutoStart to avoid this mess of magic numbers
  (Networking|SwitchHasAuthority
    (:Authority
      (bind _output_get (LocalVariables|SetCurrentMinigame (CallFunction|GetCurrentMinigame)))
      (Utilities|IsValid _output_get
        (:Is%20Valid
          (Utilities|IsValid (Variables|/Script/FortniteGame.FortMinigame|GetWarmupComponent _output_get)
            (:Is%20Valid
              (Variables|/Script/FortniteGame.FortMinigameWarmupComponent|SetAutoStartDelay (select (Variables|UserOptions|Overrides|GetAutoStart_Override) (select (!= (Math|Float|Truncate (Variables|UserOptions|GetAutoStart)) 0) (Math|Float|Truncate (Variables|UserOptions|GetAutoStart)) -1) 0) (Variables|/Script/FortniteGame.FortMinigame|GetWarmupComponent _output_get)))))))))

(fn UpdateClassSettings
  (bind _output_get (LocalVariables|SetCurrentMinigame (CallFunction|GetCurrentMinigame)))
  (Utilities|IsValid _output_get
    (:Is%20Valid
      (Utilities|IsValid (Minigame|GetFNEClassComponent _output_get)
        (:Is%20Valid
          (Variables|/Script/FortniteGame.FNEClassComponent|SetDefaultClassSlot (Creative|ClassOption|ConverttoClassSlotIndex (Variables|UserOptions|GetDefaultClassIdentifier)) (Minigame|GetFNEClassComponent _output_get))
          (Variables|/Script/FortniteGame.FNEClassComponent|SetClassResetType (Variables|UserOptions|GetRevertToDefaultClassAt) (Minigame|GetFNEClassComponent _output_get)))))))

(fn UpdateCommunicationOverride
  (bind _returnvalue (Communication|FindCreativeGameStateComponentCommunicationOverride))
  (Utilities|IsValid _returnvalue
    (:Is%20Valid
      (Class|CreativeGameStateComponentCommunicationOverride|SetCustomBubbleChatWidgetClass _returnvalue (Variables|UserOptions|GetBubbleChatCustomWidget)))))

(fn UpdateCoreExperienceSettings
  (bind _locomotionvalues (Variables|Default|GetLocomotionValues))
  (bind (_bapplyjumppenalty_7_3a31e3cc4bdc012a0163b1964bbea3df _bgliderredeploy_39_71149b6c4ba6a42bc9edb88694f6ae10 _bflyenabled_41_d15e6f05421ac053006db19a9fa624b6 _ballowflightsprint_43_adf9208b4cce05f589f691ae8ecc9d8a _flyspeedpreset_46_661d89c44c716362e5926c8bf40f5b0e) (Utilities|Struct|BreakLocomotionPresetStruct _locomotionvalues))
  (bind (_invincibility _startinghealth _maxhealth _startingshield _maxshields _dbnotype _nocooldowns _noswapcooldowns _infiniteammo _infinitemagazineammo _infinitedurability _infinitecharges _infinitereserveenergy _infiniteloadedenergy _infiniteconsumables _infinitebuildingresources _infinitegold _infiniteworldresources _disableharvestslot _displayemptyammoslots) (Class|GetClassDefaults "/FortniteCoreUI/Settings/ExperienceSettingsComponent.ExperienceSettingsComponent_C"))
  (bind (_allowfriendlyfire _vehiclesdamageobjects _vehiclesdamagevehicles) (Class|GetClassDefaults "/FortniteCoreUI/Settings/ExperienceSettingsComponent.ExperienceSettingsComponent_C"))
  (bind (_applyjumppenalty _gliderredeploy _flyenabled _allowflightsprint _flyspeed) (Class|GetClassDefaults "/FortniteCoreUI/Settings/ExperienceSettingsComponent.ExperienceSettingsComponent_C"))
  (ScalableFloat|SetValuewithContext _invincibility (Variables|UserOptions|Scoped|GetbInvincibility))
  (ScalableFloat|SetValuewithContext _maxhealth (Variables|UserOptions|SharedSettings|GetMaxHealth))
  (ScalableFloat|SetValuewithContext _startinghealth (Variables|UserOptions|SharedSettings|GetStartingHealthPercentage))
  (ScalableFloat|SetValuewithContext _maxshields (Variables|UserOptions|SharedSettings|GetMaxShields))
  (ScalableFloat|SetValuewithContext _startingshield (Variables|UserOptions|SharedSettings|GetStartingShieldPercentage))
  (ScalableFloat|SetValuewithContext _dbnotype (Variables|UserOptions|SharedSettings|GetDownButNotOut))
  (ScalableFloat|SetValuewithContext _nocooldowns (Variables|UserOptions|Scoped|GetbNoCooldowns))
  (ScalableFloat|SetValuewithContext _noswapcooldowns (Variables|UserOptions|Scoped|GetbNoCooldownsAfterSwap))
  (ScalableFloat|SetValuewithContext _infiniteammo (Variables|UserOptions|Scoped|GetbInfiniteAmmo))
  (ScalableFloat|SetValuewithContext _infinitemagazineammo (Variables|UserOptions|Scoped|GetbInfiniteMagazineAmmo))
  (ScalableFloat|SetValuewithContext _infinitedurability (Variables|UserOptions|Scoped|GetbInfiniteDurability))
  (ScalableFloat|SetValuewithContext _infinitecharges (Variables|UserOptions|Scoped|GetbInfiniteCharges))
  (ScalableFloat|SetValuewithContext _infinitereserveenergy (Variables|UserOptions|Scoped|GetbInfiniteReserveEnergy))
  (ScalableFloat|SetValuewithContext _infiniteloadedenergy (Variables|UserOptions|Scoped|GetbInfiniteLoadedEnergy))
  (ScalableFloat|SetValuewithContext _infiniteconsumables (Variables|UserOptions|Scoped|GetbInfiniteConsumables))
  (ScalableFloat|SetValuewithContext _infinitebuildingresources (Variables|UserOptions|SharedSettings|GetbInfiniteBuildingResources))
  (ScalableFloat|SetValuewithContext _infinitegold (Variables|UserOptions|SharedSettings|GetbInfiniteGold))
  (ScalableFloat|SetValuewithContext _infiniteworldresources (Variables|UserOptions|SharedSettings|GetbInfiniteWorldResources))
  (ScalableFloat|SetValuewithContext _disableharvestslot (Variables|UserOptions|Scoped|GetbDisableHarvestSlot))
  (ScalableFloat|SetValuewithContext _displayemptyammoslots (Variables|UserOptions|Scoped|GetbDisplayEmptyAmmoSlots))
  (ScalableFloat|SetValuewithContext _applyjumppenalty _bapplyjumppenalty_7_3a31e3cc4bdc012a0163b1964bbea3df)
  (ScalableFloat|SetValuewithContext _gliderredeploy _bgliderredeploy_39_71149b6c4ba6a42bc9edb88694f6ae10)
  (ScalableFloat|SetValuewithContext _flyenabled _bflyenabled_41_d15e6f05421ac053006db19a9fa624b6)
  (ScalableFloat|SetValuewithContext _allowflightsprint _ballowflightsprint_43_adf9208b4cce05f589f691ae8ecc9d8a)
  (ScalableFloat|SetValuewithContext _flyspeed _flyspeedpreset_46_661d89c44c716362e5926c8bf40f5b0e)
  (ScalableFloat|SetValuewithContext _allowfriendlyfire (Variables|UserOptions|GetbAllowFriendlyFire))
  (ScalableFloat|SetValuewithContext _vehiclesdamageobjects (Variables|UserOptions|GetbVehicleImpactsDamageObjects))
  (ScalableFloat|SetValuewithContext _vehiclesdamagevehicles (Variables|UserOptions|GetbVehicleImpactsDamageVehicles)))

(fn UpdateEndConditions
  (Networking|SwitchHasAuthority
    (:Authority
      (bind _output_get (LocalVariables|SetMinigame (CallFunction|GetCurrentMinigame)))
      (Utilities|IsValid _output_get
        (:Is%20Valid
          (Utilities|FlowControl|Sequence
            (:then_0
              (bind _outputpin (Utilities|AddEndConditionifValueIsGTZ (select (Variables|UserOptions|Overrides|GetEliminationsToEnd_Override) (Variables|UserOptions|GetEliminationsToEnd) 0) (LocalVariables|GetEndConditions) "/Game/Athena/Playset/Minigames/StatFilters/MinigameStat_PlayerKills.MinigameStat_PlayerKills_C"))
              (bind _outputpin_2 (Utilities|AddEndConditionifValueIsGTZ (select (Variables|UserOptions|Overrides|GetAIEnemyEliminationsToEnd_Override) (Variables|UserOptions|GetAIEnemyEliminationsToEnd) 0) _outputpin "/Game/Athena/Playset/Minigames/StatFilters/MinigameStat_AIKills.MinigameStat_AIKills_C"))
              (bind _outputpin_3 (Utilities|AddEndConditionfromStatFilterifValueIsGTZ (select (Variables|UserOptions|Overrides|GetStatToEnd_Override) (Variables|UserOptions|GetStatValueToEnd) 0) _outputpin_2 (Variables|UserOptions|GetStatToEnd)))
              (CallFunction|TryAddEndConditionCollectItems _outputpin_3)
              (CallFunction|TryAddEndConditionDestroyObjects _outputpin_3))
            (:then_1
              (Variables|/Script/FortniteGame.FortMinigameEndConditionComponent|SetBaseEndConditionArray (LocalVariables|GetEndConditions) (Variables|/Script/FortniteGame.FortMinigame|GetEndConditionComponent _output_get)))))))))

(fn UpdateFallDamageType
  (CallFunction|LoadandApplyCurveTabletoRegistry (Utilities|Select :Index (Utilities|Struct|BreakLocomotionPresetStruct (Variables|Default|GetLocomotionValues)) :NewEnumerator1 "/CreativeCoreDevices/ExperienceSettings/FallingDamage_Creative.FallingDamage_Creative" :NewEnumerator4 "/CreativeCoreDevices/ExperienceSettings/FallingDamageV2_Creative.FallingDamageV2_Creative")))

(fn UpdateGameMode
  (bind _localfneuicomponent (LocalVariables|GetLocalFNEUIComponent))
  (CallFunction|FillTrackedStats)
  (bind _output_get (LocalVariables|SetCurrentMinigame (CallFunction|GetCurrentMinigame)))
  (Utilities|IsValid _output_get
    (:Is%20Valid
      (Utilities|IsValid (Minigame|GetFNEUIComponent _output_get)
        (:Is%20Valid
          (LocalVariables|SetLocalFNEUIComponent (Minigame|GetFNEUIComponent _output_get))
          (Utilities|FlowControl|Sequence
            (:then_0
              (CallFunction|UpdateMinigameStatData (Variables|/Script/FortniteGame.FortMinigame|GetVolume _output_get))
              (Class|FNEUIComponent|SetUIExtensionTags _localfneuicomponent (CallFunction|GetUIExtensions (Minigame|IsCuratedHub _output_get)) :!class /Script/FortniteGame.FNEUIComponent)
              (Class|FNEUIComponent|SetMinigameMapWidget _localfneuicomponent (Variables|UserOptions|GetMapScreenDisplay) :!class /Script/FortniteGame.FNEUIComponent)
              (Class|FNEUIComponent|SetShowMapWidget _localfneuicomponent (Variables|UserOptions|GetbDisplayOverviewMap) :!class /Script/FortniteGame.FNEUIComponent)
              (Variables|/Script/FortniteGame.FortMinigame|SetWinCondition (Variables|UserOptions|GetGameWinCondition) _output_get)
              (Variables|/Script/FortniteGame.FortMinigameEndConditionComponent|SetbAllTeamsMustMatchEndConditions (Variables|UserOptions|GetbAllTeamsMustFinish) (Variables|/Script/FortniteGame.FortMinigame|GetEndConditionComponent _output_get))
              (Variables|/Script/FortniteGame.FortMinigame|SetbEndGameOnMatchPointWin (Variables|UserOptions|GetbEndGameOnMatchPointWin) _output_get)
              (Variables|/Script/FortniteGame.FortMinigame|SetPlayerPersistence (Variables|Default|GetPlayerPersistence) _output_get)
              (Variables|/Script/FortniteGame.FortMinigame|SetbOnlyAllowRespawningIfPlayerStartPadsFound (Variables|UserOptions|GetbOnlyAllowRespawnIfSpawnPadsFound) _output_get)
              (Class|FNEUIComponent|SetAllowTeamIndicators (LocalVariables|GetLocalFNEUIComponent) (Variables|UserOptions|GetbAllowTeamIndicators) :!class /Script/FortniteGame.FNEUIComponent))
            (:then_1
              (bind _scoreboardcomponent (Variables|/Script/FortniteGame.FortMinigame|GetScoreboardComponent _output_get)
                (:then
                  (Variables|/Script/FortniteGame.FortMinigameScoreboardComponent|SetbSortScoreboardEntries true _scoreboardcomponent)
                  (Variables|/Script/FortniteGame.FortMinigameScoreboardComponent|SetbShowCumulativeScoreboard (Variables|UserOptions|GetbShowCumulativeScoreboard) _scoreboardcomponent)
                  (Variables|/Script/FortniteGame.FortMinigameScoreboardComponent|SetbForceIndividualScoresInTeamGame (Variables|UserOptions|GetbShowIndividualScores) _scoreboardcomponent)
                  (Variables|/Script/FortniteGame.FortMinigameScoreboardComponent|SetbIncludePlayersWhoLeftGameInScoreboard (Variables|UserOptions|GetbIncludePlayersWhoLeftGameInScoreboard) _scoreboardcomponent))
                (:else)))))))))

(fn UpdateGameSettings
  (bind _currentminigame (LocalVariables|GetCurrentMinigame))
  (bind _victoryscreencomponent (Variables|/Script/FortniteGame.FortMinigame|GetVictoryScreenComponent _currentminigame))
  (bind _output_get (LocalVariables|SetCurrentMinigame (CallFunction|GetCurrentMinigame)))
  (Utilities|IsValid _output_get
    (:Is%20Valid
      (Utilities|FlowControl|Sequence
        (:then_0
          (bind (_victorycallout _defaultcallout _gameendcallout) (UI|GetGameEndCallouts true :!class /Game/Creative/Devices/CreativeMinigameUIInterface.CreativeMinigameUIInterface_C))
          (Variables|UserOptions|SetGameEndCallout _gameendcallout)
          ;[1 0 0.06958] TODO: Migrate Sound Present options to proper Asset References
          (Variables|/Script/FortniteGame.FortMinigameVictoryScreenComponent|SetGameEndCallout (Variables|UserOptions|GetGameEndCallout) _victoryscreencomponent)
          (Variables|/Script/FortniteGame.FortMinigameVictoryScreenComponent|SetVictoryAudioIndex (Math|Conversions|ToInteger(Byte) (Variables|UserOptions|GetVictorySound)) _victoryscreencomponent)
          (Variables|/Script/FortniteGame.FortMinigameVictoryScreenComponent|SetDrawAudioIndex (Math|Conversions|ToInteger(Byte) (Variables|UserOptions|GetDrawSound)) _victoryscreencomponent))
        (:then_1
          (Variables|/Script/FortniteGame.FortMinigameVictoryScreenComponent|SetDefeatAudioIndex (Math|Conversions|ToInteger(Byte) (Variables|UserOptions|GetDefeatSound)) _victoryscreencomponent))
        (:then_2
          (Variables|/Script/FortniteGame.FortMinigameVictoryScreenComponent|SetPostGameSettings (Utilities|Struct|MakeMinigamePostGameSettings (Variables|UserOptions|GetPostGameType) 0.0 (Variables|UserOptions|GetbPostGameCustomShowScoreboard) (Utilities|Struct|MakeMinigamePostGameAnimationSettings (Variables|UserOptions|GetPostGameCustomVictoryAnimationStyle) (Variables|UserOptions|GetPostGameCustomVictoryAnimationColorSet) (Variables|UserOptions|GetPostGameCustomVictoryText) (Variables|UserOptions|GetPostGameCustomVictorySubText)) (Utilities|Struct|MakeMinigamePostGameAnimationSettings (Variables|UserOptions|GetPostGameCustomDefeatAnimationStyle) (Variables|UserOptions|GetPostGameCustomDefeatAnimationColorSet) (Variables|UserOptions|GetPostGameCustomDefeatText) (Variables|UserOptions|GetPostGameCustomDefeatSubText)) (Utilities|Struct|MakeMinigamePostGameAnimationSettings (Variables|UserOptions|GetPostGameCustomTieAnimationStyle) (Variables|UserOptions|GetPostGameCustomTieAnimationColorSet) (Variables|UserOptions|GetPostGameCustomTieText) (Variables|UserOptions|GetPostGameCustomTieSubText))) (Variables|/Script/FortniteGame.FortMinigame|GetVictoryScreenComponent _currentminigame)))))))

(fn UpdateGameStartCountdown
  (bind _minigame (CallFunction|GetCurrentMinigame))
  (bind _warmupcomponent (Variables|/Script/FortniteGame.FortMinigame|GetWarmupComponent _minigame))
  (Networking|SwitchHasAuthority
    (:Authority
      (Utilities|IsValid _minigame
        (:Is%20Valid
          (Utilities|IsValid _warmupcomponent
            (:Is%20Valid
              (Variables|/Script/FortniteGame.FortMinigameWarmupComponent|SetWarmupDuration (Variables|UserOptions|GetGameStartCountdown) _warmupcomponent)
              (Class|FortMinigameWarmupComponent|SetGameStartSoundEnabled _warmupcomponent (Variables|UserOptions|GetbGameStartSoundEnabled) :!class /Script/FortniteGame.FortMinigameWarmupComponent))))))))

(fn UpdateJamSettings
  (ScalableFloat|SetValuewithContext (Variables|Default|GetbEnableJam) (CallFunction|GetJamEnabledOption (Variables|UserOptions|GetEnableJam))))

(fn UpdateJoinInProgress
  (Networking|SwitchHasAuthority
    (:Authority
      (bind _output_get (LocalVariables|SetNewLocalVar_1 (CallFunction|GetCurrentMinigame)))
      (Utilities|IsValid _output_get
        (:Is%20Valid
          (Variables|/Script/FortniteGame.FortMinigame|SetbAllowJoinInProgress (Utilities|Enum|NotEqual(Enum) (Variables|UserOptions|GetJoinInProgressBehavior) "WatchOnly") _output_get))))))

(fn UpdateLedgeLaunchingSettings
  (bind (_ballowmantling_9_2d7792d74b333533d42486b460884d0d _bpbwsgenerateledges_11_337764964a1b07662e899aa89a26263f) (Utilities|Struct|BreakLocomotionPresetStruct (Variables|Default|GetLocomotionValues)))
  (bind _returnvalue (WorldSubsystems|GetBuildingWallSubsystem))
  (Utilities|IsValid _returnvalue
    (:Is%20Valid
      (Class|BuildingWallSubsystem|UpdateLedgeLaunchingData _returnvalue self (and _ballowmantling_9_2d7792d74b333533d42486b460884d0d _bpbwsgenerateledges_11_337764964a1b07662e899aa89a26263f)))))

(fn UpdateLocomotionSettings
  (bind _locomotionvalues (Variables|Default|GetLocomotionValues))
  (bind (_ballowboostedjump_55_d7cd042547e6720fc6558191fb842532 _ballowrolllanding_57_0c837ec94f51848ac25f9e8693db244c _rolllandingenergyrechargefatigue_136_c11d8ca24cab12981f77788363f8ff04 _ballowwallkick_59_360afa9b4861125509db57a13e811574 _ballowwallscramble_61_1ff1ad0946ce324f6099a7b54f01e691) (Utilities|Struct|BreakLocomotionPresetStruct _locomotionvalues))
  (bind (_rolllandingenergyrechargefatigue_136_c11d8ca24cab12981f77788363f8ff04_2 _ballowwallkick_59_360afa9b4861125509db57a13e811574_2 _ballowwallscramble_61_1ff1ad0946ce324f6099a7b54f01e691_2) (Utilities|Struct|BreakLocomotionPresetStruct (Variables|Default|GetLocomotionValues)))
  (bind _locomotionvalues_2 (Variables|Default|GetLocomotionValues))
  (bind (_bfalldamage_1_11e1610b4a1446a10b6c9693ae244fef _bshouldcapfalldamage_85_b3be7a61451a54d1a3124e8538e046ad _ballowgroundrush_86_b788f2a546029e9618308abac5cbbe81) (Utilities|Struct|BreakLocomotionPresetStruct _locomotionvalues_2))
  (bind _locomotionvalues_3 (Variables|Default|GetLocomotionValues))
  (bind (_ballowmantling_9_2d7792d74b333533d42486b460884d0d _ballowvaulting_22_2035cf0f4b9dcc4ba5a877af068194d5 _rolllandingenergyrechargefatigue_136_c11d8ca24cab12981f77788363f8ff04_3 _bhurdleonlywhenjumpblocked_75_add25a364067fe7c7bd999964acc78ac) (Utilities|Struct|BreakLocomotionPresetStruct _locomotionvalues_3))
  (bind _locomotionvalues_4 (Variables|Default|GetLocomotionValues))
  (bind (_ballowsprinting_23_30bcf21447b1ae2a9b8400a9a6dd74ae _sprintingenergycostpersecond_26_5efd167445c4bf5add1922a981a67e8c _maxsprintingjumpmultiplier_28_72abf30642d1c5cdfe9c158b5dd345b2 _maxsprintingspeedmultiplier_30_7d1fd1024020fdf09b7efa85b01ff381) (Utilities|Struct|BreakLocomotionPresetStruct _locomotionvalues_4))
  (bind _locomotionvalues_5 (Variables|Default|GetLocomotionValues))
  (bind (_ballowsliding_33_f2350fb54315154e33ef6f9e39329ed0 _ballowslidekick_35_e880fdbf48f7627124a390904aaafba5 _ballowshoulderbashing_37_30c279db4e82f248338b8aa94166b1f7) (Utilities|Struct|BreakLocomotionPresetStruct _locomotionvalues_5))
  (bind _locomotionvalues_6 (Variables|Default|GetLocomotionValues))
  (bind (_rolllandingenergyrechargefatigue_136_c11d8ca24cab12981f77788363f8ff04_4 _energycostonjump_78_5a70fcd54a55c74db172a4b285cc949c _bshouldpauseenergyuseonfalling_117_cf8b27dd464c97d0d4d7498fc9a958bf) (Utilities|Struct|BreakLocomotionPresetStruct _locomotionvalues_6))
  (bind _locomotionvalues_7 (Variables|Default|GetLocomotionValues))
  (bind (_energymax_70_ace05d3b4287ca64b9cdb896123d06ea _energyrechargeamount_68_e905827e4fa43c44b8ba25ba1cc19e40 _energyrechargedelay_69_3c4ad7a94a3cf317dc7c1e8d9e25bb7b) (Utilities|Struct|BreakLocomotionPresetStruct _locomotionvalues_7))
  (bind _localminigame (LocalVariables|GetLocalMinigame))
  (bind _volume (Variables|/Script/FortniteGame.FortMinigame|GetVolume _localminigame))
  ; Set Shared Settings
  (ScalableFloat|SetValuewithContext (struct :Value 1.0 :Curve (struct :CurveTable 0 :RowName "BoostedJump.EnergyCost") :RegistryType "Fort.DataRegistry") (Utilities|Select :Index (Variables|UserOptions|GetLocomotionPreset) :NewEnumerator0 15.0 :NewEnumerator1 0.0 :NewEnumerator2 15.0) true)
  (ScalableFloat|SetValuewithContext (struct :Value 1.0 :Curve (struct :CurveTable 0 :RowName "FallDamageOverride.Enabled") :RegistryType "Fort.DataRegistry") (select _bfalldamage_1_11e1610b4a1446a10b6c9693ae244fef _bshouldcapfalldamage_85_b3be7a61451a54d1a3124e8538e046ad false) true)
  (ScalableFloat|SetValuewithContext (struct :Value 1.0 :Curve (struct :CurveTable 0 :RowName "GliderRelease.Enabled") :RegistryType "Fort.DataRegistry") _ballowgroundrush_86_b788f2a546029e9618308abac5cbbe81 true)
  (ScalableFloat|SetValuewithContext (struct :Value 1.0 :Curve (struct :CurveTable 0 :RowName "Default.TacticalSprint.Energy.Max") :RegistryType "Fort.DataRegistry") _energymax_70_ace05d3b4287ca64b9cdb896123d06ea true)
  (ScalableFloat|SetValuewithContext (struct :Value 1.0 :Curve (struct :CurveTable 0 :RowName "Default.TacticalSprint.Energy.RechargeAmountPerSecond") :RegistryType "Fort.DataRegistry") _energyrechargeamount_68_e905827e4fa43c44b8ba25ba1cc19e40 true)
  (ScalableFloat|SetValuewithContext (struct :Value 1.0 :Curve (struct :CurveTable 0 :RowName "Default.TacticalSprint.Energy.RechargeDelayInSeconds") :RegistryType "Fort.DataRegistry") _energyrechargedelay_69_3c4ad7a94a3cf317dc7c1e8d9e25bb7b true)
  (ScalableFloat|SetValuewithContext (struct :Value 1.0 :Curve (struct :CurveTable 0 :RowName "ClamberingContent.Clambering.Enabled") :RegistryType "Fort.DataRegistry") _ballowmantling_9_2d7792d74b333533d42486b460884d0d true)
  (ScalableFloat|SetValuewithContext (struct :Value 1.0 :Curve (struct :CurveTable 0 :RowName "Default.VaultingContent.Hurdle.Enabled") :RegistryType "Fort.DataRegistry") _ballowvaulting_22_2035cf0f4b9dcc4ba5a877af068194d5 true)
  (ScalableFloat|SetValuewithContext (struct :Value 1.0 :Curve (struct :CurveTable 0 :RowName "Default.VaultingContent.Hurdle.HurdleOnlyWhenJumpBlocked") :RegistryType "Fort.DataRegistry") _bhurdleonlywhenjumpblocked_75_add25a364067fe7c7bd999964acc78ac true)
  (ScalableFloat|SetValuewithContext (struct :Value 1.0 :Curve (struct :CurveTable 0 :RowName "Default.TacticalSprint.Enabled") :RegistryType "Fort.DataRegistry") _ballowsprinting_23_30bcf21447b1ae2a9b8400a9a6dd74ae true)
  (ScalableFloat|SetValuewithContext (struct :Value 1.0 :Curve (struct :CurveTable 0 :RowName "Default.TacticalSprint.Sprint.Energy.CostPerSecond") :RegistryType "Fort.DataRegistry") _sprintingenergycostpersecond_26_5efd167445c4bf5add1922a981a67e8c true)
  (ScalableFloat|SetValuewithContext (struct :Value 1.0 :Curve (struct :CurveTable 0 :RowName "Default.TacticalSprint.JumpZMultiplierTotal") :RegistryType "Fort.DataRegistry") _maxsprintingjumpmultiplier_28_72abf30642d1c5cdfe9c158b5dd345b2 true)
  (ScalableFloat|SetValuewithContext (struct :Value 1.0 :Curve (struct :CurveTable 0 :RowName "Default.TacticalSprint.SpeedMultiplierTotal") :RegistryType "Fort.DataRegistry") _maxsprintingspeedmultiplier_30_7d1fd1024020fdf09b7efa85b01ff381 true)
  (ScalableFloat|SetValuewithContext (struct :Value 1.0 :Curve (struct :CurveTable 0 :RowName "Default.Sliding.Enabled") :RegistryType "Fort.DataRegistry") _ballowsliding_33_f2350fb54315154e33ef6f9e39329ed0 true)
  (ScalableFloat|SetValuewithContext (struct :Value 1.0 :Curve (struct :CurveTable 0 :RowName "Default.SlideImpulse.Enabled") :RegistryType "Fort.DataRegistry") _ballowslidekick_35_e880fdbf48f7627124a390904aaafba5 true)
  (ScalableFloat|SetValuewithContext (struct :Value 1.0 :Curve (struct :CurveTable 0 :RowName "Default.DoorBashContent.DoorBash.Enabled") :RegistryType "Fort.DataRegistry") _ballowshoulderbashing_37_30c279db4e82f248338b8aa94166b1f7 true)
  (ScalableFloat|SetValuewithContext (struct :Value 1.0 :Curve (struct :CurveTable 0 :RowName "BoostedJump.Enabled") :RegistryType "Fort.DataRegistry") _ballowboostedjump_55_d7cd042547e6720fc6558191fb842532 true)
  (ScalableFloat|SetValuewithContext (struct :Value 1.0 :Curve (struct :CurveTable 0 :RowName "RollLanding.Enabled") :RegistryType "Fort.DataRegistry") _ballowrolllanding_57_0c837ec94f51848ac25f9e8693db244c true)
  (ScalableFloat|SetValuewithContext (struct :Value 1.0 :Curve (struct :CurveTable 0 :RowName "WallJump.Enabled") :RegistryType "Fort.DataRegistry") _ballowwallkick_59_360afa9b4861125509db57a13e811574 true)
  (ScalableFloat|SetValuewithContext (struct :Value 1.0 :Curve (struct :CurveTable 0 :RowName "WallStick.Enabled") :RegistryType "Fort.DataRegistry") _ballowwallscramble_61_1ff1ad0946ce324f6099a7b54f01e691 true)
  (ScalableFloat|SetValuewithContext (struct :Value 1.0 :Curve (struct :CurveTable 0 :RowName "Default.TacticalSprint.Sprint.Energy.CostOnJump") :RegistryType "Fort.DataRegistry") _energycostonjump_78_5a70fcd54a55c74db172a4b285cc949c true)
  (ScalableFloat|SetValuewithContext (struct :Value 1.0 :Curve (struct :CurveTable 0 :RowName "Default.TacticalSprint.Energy.ShouldPauseEnergyUseOnFalling") :RegistryType "Fort.DataRegistry") _bshouldpauseenergyuseonfalling_117_cf8b27dd464c97d0d4d7498fc9a958bf true)
  (ScalableFloat|SetValuewithContext (struct :Value 1.0 :Curve (struct :CurveTable 0 :RowName "WallJump.EnabledFromWallStick") :RegistryType "Fort.DataRegistry") (and _ballowwallkick_59_360afa9b4861125509db57a13e811574_2 _ballowwallscramble_61_1ff1ad0946ce324f6099a7b54f01e691_2) true)
  (bind _output_get (LocalVariables|SetLocalMinigame (CallFunction|GetCurrentMinigame)))
  (Utilities|IsValid _output_get
    (:Is%20Valid
      (Utilities|FlowControl|Sequence
        (:then_0
          ;[1 1 1] Update player collision mutator
          (bind _output_get_2 (LocalVariables|SetMinigameMutatorList (Actor|GetComponentbyClass (LocalVariables|GetLocalMinigame) "/Script/FortniteGame.FortMutatorListComponent")))
          (Utilities|IsValid _output_get_2
            (:Is%20Valid
              (bind _ascreative_mutator_disable_pawn_collisions (Utilities|Casting|CastToCreativeMutator_DisablePawnCollisions (Modifier|GetMutatorbyClass _output_get_2 "/Game/Athena/Playlists/Creative/Mutators/CreativeMutator_DisablePawnCollisions.CreativeMutator_DisablePawnCollisions_C")))
              (Variables|/Script/FortniteGame.FortAthenaMutator_GameplayEffect|SetbEnableMutator (Utilities|Struct|BreakLocomotionPresetStruct (Variables|Default|GetLocomotionValues)) _ascreative_mutator_disable_pawn_collisions))))
        (:then_1
          ;[1 1 1] Update fall damage mutator - Could be converted to shared setting eventually.
          (Utilities|IsValid _volume
            (:Is%20Valid
              (bind _output_get_3 (LocalVariables|SetVolumeMutatorList (Actor|GetComponentbyClass _volume "/Script/FortniteGame.FortMutatorListComponent")))
              (Utilities|IsValid _output_get_3
                (:Is%20Valid
                  (bind _ascreative_mutator_fall_damage (Utilities|Casting|CastToCreativeMutator_FallDamage (Modifier|GetMutatorbyClass _output_get_3 "/Game/Athena/Playlists/Creative/Mutators/CreativeMutator_FallDamage.CreativeMutator_FallDamage_C")))
                  (Variables|/Script/FortniteGame.FortAthenaMutator_FallDamageMultiplier|SetbFallDamage (Utilities|Struct|BreakLocomotionPresetStruct (Variables|Default|GetLocomotionValues)) _ascreative_mutator_fall_damage))))))))))

(fn UpdateMantlingSettings
  (bind _locomotionvalues (Variables|Default|GetLocomotionValues))
  (bind (_mantlingminimumheight_14_57a92d904d139b6d337647b53d1c7bb3 _mantlingminimumheightinwater_16_027fcf034c541ec5cceaafa2a1989000) (Utilities|Struct|BreakLocomotionPresetStruct _locomotionvalues))
  (ScalableFloat|SetValuewithContext (Variables|Constants|Mantling|GetMantlingHeightSetting) (Mantling|MantlingHeightPresettoFloat (Variables|Constants|Mantling|GetMantlingHeightMin) (Variables|Constants|Mantling|GetMantlingHeightMax) (Variables|Constants|Mantling|GetMantlingHeightIncrement) _mantlingminimumheight_14_57a92d904d139b6d337647b53d1c7bb3))
  (ScalableFloat|SetValuewithContext (Variables|Constants|Mantling|GetMantlingHeightWaterSetting) (Mantling|MantlingHeightPresettoFloat (Variables|Constants|Mantling|GetMantlingHeightWaterMin) (Variables|Constants|Mantling|GetMantlingHeightWaterMax) (Variables|Constants|Mantling|GetMantlingHeightWaterIncrement) _mantlingminimumheightinwater_16_027fcf034c541ec5cceaafa2a1989000)))

(fn UpdateMinigameMutators
  (Networking|SwitchHasAuthority
    (:Authority
      (bind _output_get (LocalVariables|SetCurrentMinigame (CallFunction|GetCurrentMinigame)))
      (Utilities|IsValid _output_get
        (:Is%20Valid
          (bind _output_get_2 (LocalVariables|SetMutatorList (Actor|GetComponentbyClass _output_get "/Script/FortniteGame.FortMutatorListComponent")))
          (Modifiers|SetPropertyOverrides _output_get_2 (OptionData|GetPropertyOverrides (Variables|Default|GetToyOptionsComponent))))))))

(fn UpdateMinigameStatData
  (params (ContextActor :type Actor%20Object%20Reference))
  (Utilities|IsValid ContextActor
    (:Is%20Valid
      (bind _output_get (Variables|Default|SetCachedMinigameStatsComponent (Minigame|Stats|GetMinigameStatsComponent ContextActor)))
      (Utilities|IsValid _output_get
        (:Is%20Valid
          (Minigame|Stats|SetTrackedStats _output_get (Variables|Default|GetTrackedStats))
          (bind _outarray (CallFunction|StatFilterStructtoStatFilterObjectArray (Variables|Default|GetTrackedScoreStatFilters)))
          (Variables|/Script/FortniteGame.FortMinigameStatsComponent|SetScoreboardStatObjects _outarray _output_get)
          (bind _outarray_2 (CallFunction|StatFilterStructtoStatFilterObjectArray (Variables|Default|GetTrackedScoreColumnOrder)))
          (Minigame|Stats|SetScoreboardColumnOrder _output_get _outarray_2)
          (Minigame|Stats|SetTimeAliveTeamTrackingMethod _output_get (Variables|UserOptions|GetTimeAliveTeamTrackingMethod))
          (Minigame|Stats|SetTimeAliveStartPoint _output_get (Variables|UserOptions|GetTimeAliveStartPoint)))))))

(fn UpdateMovementSpeedOverrides
  (bind _locomotionvalues (Variables|Default|GetLocomotionValues))
  (bind _movementspeedtunings_52_c64aef264600e54ed15a00b564f100b5 (Utilities|Struct|BreakLocomotionPresetStruct _locomotionvalues))
  (bind _movementoverridedatasettags (Variables|Movement|GetMovementOverrideDataSetTags))
  (bind (_value _returnvalue) (Utilities|Map|Find _movementoverridedatasettags _movementspeedtunings_52_c64aef264600e54ed15a00b564f100b5))
  (bind _returnvalue_2 (WorldSubsystems|GetFortPawnMovementOverridesSubsystem))
  (Utilities|IsValid _returnvalue_2
    (:Is%20Valid
      (MovementSet|RemoveMovementOverrideDatabyTag _returnvalue_2 (Variables|Movement|GetCurrentMovementOverrideDataSetTag) self)
      (if _returnvalue
        (MovementSet|ApplyMovementOverrideDatabyTag (WorldSubsystems|GetFortPawnMovementOverridesSubsystem) _value self 0.0)
        (Variables|Movement|SetCurrentMovementOverrideDataSetTag _value)))))

(fn UpdatePatchworkSettings
  (ScalableFloat|SetValuewithContext (Variables|Default|GetPatchworkMemoryMode) (Variables|UserOptions|GetbPatchworkMemoryMode)))

(fn UpdatePhysicsSettings
  (bind _benablecoresettings_ch4s2 (Variables|Default|GetbEnableCoreSettings_Ch4S2))
  (ScalableFloat|SetValuewithContext (Variables|Default|GetPhysicsUserDataSetupSetting) _benablecoresettings_ch4s2)
  (ScalableFloat|SetValuewithContext (Variables|Default|GetPropToPropPhysicsSetting) _benablecoresettings_ch4s2)
  (ScalableFloat|SetValuewithContext (Variables|Default|GetVehicleToPropPhysicsSetting) _benablecoresettings_ch4s2)
  (CallFunction|UpdateGravityValue (Variables|UserOptions|Overrides|GetGravityValue_Override) (Variables|UserOptions|GetGravityValue)))

(fn UpdatePlayerLives
  (Networking|SwitchHasAuthority
    (:Authority
      (bind _output_get (LocalVariables|SetNewLocalVar_0 (CallFunction|GetCurrentMinigame)))
      (Utilities|IsValid _output_get
        (:Is%20Valid
          (Variables|/Script/FortniteGame.FortMinigame|SetMaxLivesForPlayer (select (Variables|UserOptions|Overrides|GetSpawnLimit_Override) (Variables|UserOptions|GetSpawnLimit) 0) _output_get))))))

(fn UpdatePlayerPersistence
  (bind _minigame (CallFunction|GetCurrentMinigame))
  (Networking|SwitchHasAuthority
    (:Authority
      (Utilities|IsValid _minigame
        (:Is%20Valid
          (Variables|/Script/FortniteGame.FortMinigame|SetPlayerPersistence (Variables|Default|GetPlayerPersistence) _minigame))))))

(fn UpdatePostGameSpawnLocationSetting
  (bind _minigame (CallFunction|GetCurrentMinigame))
  (Networking|SwitchHasAuthority
    (:Authority
      (LocalVariables|SetCurrentMinigame _minigame)
      (Utilities|IsValid _minigame
        (:Is%20Valid
          (Variables|/Script/FortniteGame.FortMinigame|SetPostGameSpawnLocationSetting (Variables|UserOptions|GetPostGameSpawnLocation) _minigame))))))

(fn UpdateSlowMotionOnEndOfRound
  (Networking|SwitchHasAuthority
    (:Authority
      (bind _output_get (LocalVariables|SetMinigame (CallFunction|GetCurrentMinigame)))
      (Utilities|IsValid _output_get
        (:Is%20Valid
          (Variables|/Script/FortniteGame.FortMinigame|SetbSlowMotionOnEndOfRound (Variables|UserOptions|GetbSlowMotionOnEndOfRoundV2) _output_get))))))

(fn UpdateSpawnLocationSetting
  (bind _minigame (CallFunction|GetCurrentMinigame))
  (Networking|SwitchHasAuthority
    (:Authority
      (LocalVariables|SetCurrentMinigame _minigame)
      (Utilities|IsValid _minigame
        (:Is%20Valid
          (Variables|/Script/FortniteGame.FortMinigame|SetSpawnLocationSetting (Variables|UserOptions|GetSpawnLocation) _minigame))))))

(fn UpdateSpawnPadSelection
  (bind _minigame (CallFunction|GetCurrentMinigame))
  (Networking|SwitchHasAuthority
    (:Authority
      (LocalVariables|SetMinigame _minigame)
      (Utilities|IsValid _minigame
        (:Is%20Valid
          (Variables|/Script/FortniteGame.FortMinigame|SetSpawnPadSelection (Variables|UserOptions|GetSpawnPadSelection) _minigame))))))

(fn UpdateSpectatorSettings
  (bind _returnvalue (Class|PlayspaceComponentSpectateMode|GetPlayspaceComponent_SpectateMode self false))
  (Utilities|IsValid _returnvalue
    (:Is%20Valid
      (switch Utilities|FlowControl|Switch|SwitchonECreativeAllowSpectateOtherTeams (Variables|UserOptions|Scoped|GetAllowSpectateOtherTeams)
        (:Disallow
          (bind _structout (Utilities|Struct|SetmembersinSpectatorModeGeneralOptions (Class|PlayspaceComponentSpectateMode|GetSpectateOptions _returnvalue 0) false true true false false))
          (Class|PlayspaceComponentSpectateMode|SetDefaultSpectateOptions _returnvalue _structout))
        (:Allow
          (bind _structout_2 (Utilities|Struct|SetmembersinSpectatorModeGeneralOptions (Class|PlayspaceComponentSpectateMode|GetSpectateOptions _returnvalue 0) true true true true true))
          (Class|PlayspaceComponentSpectateMode|SetDefaultSpectateOptions _returnvalue _structout_2))
        (:BattleRoyale
          (bind _structout_3 (Utilities|Struct|SetmembersinSpectatorModeGeneralOptions (Class|PlayspaceComponentSpectateMode|GetSpectateOptions _returnvalue 0) true true true false false))
          (Class|PlayspaceComponentSpectateMode|SetDefaultSpectateOptions _returnvalue _structout_3))
        (:None
          (bind _structout_4 (Utilities|Struct|SetmembersinSpectatorModeGeneralOptions (Class|PlayspaceComponentSpectateMode|GetSpectateOptions _returnvalue 0) false false false false false))
          (Class|PlayspaceComponentSpectateMode|SetDefaultSpectateOptions _returnvalue _structout_4))))))

(fn UpdateTeamSetting
  (bind _returnvalue (Minigame|GetMinigameTeamComponent (CallFunction|GetCurrentMinigame)))
  (bind _minigame (CallFunction|GetCurrentMinigame))
  (bind _returnvalue_2 (Minigame|GetMinigameTeamBalanceComponent _minigame))
  (Networking|SwitchHasAuthority
    (:Authority
      (Utilities|IsValid _minigame
        (:Is%20Valid
          (Utilities|FlowControl|Sequence
            (:then_0
              (Utilities|IsValid _returnvalue
                (:Is%20Valid
                  (Minigame|Teams|SetMinigameNumTeams _returnvalue (CallFunction|GetTeamSetting)))))
            (:then_1
              (Utilities|IsValid _returnvalue_2
                (:Is%20Valid
                  (Variables|/Script/FortniteGame.FortMinigameTeamBalanceComponent|SetTeamRotationSetting (select (Variables|UserOptions|Overrides|GetTeamRotation_Override) (Variables|UserOptions|GetTeamRotation) 0) _returnvalue_2)
                  (Minigame|Teams|SetStableTeamCosmetics _returnvalue_2 (Utilities|Enum|Equal(Enum) (Variables|UserOptions|GetTeamVisualsDeterminedAt) "NewEnumerator3")))))))))))

(fn UpdateTeamToMoveToWhenOutOfSpawns
  (Networking|SwitchHasAuthority
    (:Authority
      (bind _output_get (LocalVariables|SetMinigame (CallFunction|GetCurrentMinigame)))
      (Utilities|IsValid _output_get
        (:Is%20Valid
          (Variables|/Script/FortniteGame.FortMinigame|SetTeamToSwapToWhenOutOfSpawns (Creative|TeamOption|ConverttoTeam (Variables|UserOptions|GetAfterLastSpawnGoTo)) _output_get))))))

(fn UpdateTestPlayers
  (Default|UpdateTestPlayers (Variables|Default|GetTestPlayerDefinitionComponent)))

(fn UpdateTimeLimit
  (Networking|SwitchHasAuthority
    (:Authority
      (bind _output_get (LocalVariables|SetCurrentMinigame (CallFunction|GetCurrentMinigame)))
      (Utilities|IsValid _output_get
        (:Is%20Valid
          (Utilities|FlowControl|Sequence
            (:then_0
              (Variables|/Script/FortniteGame.FortMinigame|SetTimeLimit (select (Variables|UserOptions|Overrides|GetRoundTimeLimit_Override) (* (Variables|UserOptions|GetRoundTimeLimit) 6e+01) 0.0) _output_get)
              (Variables|/Script/FortniteGame.FortMinigame|SetbTimerCountsDown (Utilities|Enum|Equal(Enum) (Variables|UserOptions|GetTimerDirection) "NewEnumerator0") _output_get)
              (Variables|/Script/FortniteGame.FortMinigame|SetbStopwatchMode (Variables|UserOptions|GetbFastestTimeWin) _output_get))
            (:then_1
              (bind _scoreboardcomponent (Variables|/Script/FortniteGame.FortMinigame|GetScoreboardComponent _output_get)
                (:then
                  (Variables|/Script/FortniteGame.FortMinigameScoreboardComponent|SetGameWinnerDisplayTime (Variables|UserOptions|GetGameWinnerDisplayTime) _scoreboardcomponent)
                  (Variables|/Script/FortniteGame.FortMinigameScoreboardComponent|SetGameScoreDisplayTime (Variables|UserOptions|GetGameScoreDisplayTime) _scoreboardcomponent)
                  (Variables|/Script/FortniteGame.FortMinigameScoreboardComponent|SetRoundWinnerDisplayTime (Variables|UserOptions|GetRoundWinnerDisplayTime) _scoreboardcomponent)
                  (Variables|/Script/FortniteGame.FortMinigameScoreboardComponent|SetRoundScoreDisplayTime (Variables|UserOptions|GetRoundScoreDisplayTime) _scoreboardcomponent))
                (:else)))))))))

(fn UpdateTotalRounds
  (bind _minigame (CallFunction|GetCurrentMinigame))
  (bind _returnvalue (Minigame|GetFNEUIComponent _minigame))
  (Networking|SwitchHasAuthority
    (:Authority
      (Utilities|IsValid _minigame
        (:Is%20Valid
          (Utilities|IsValid _returnvalue
            (:Is%20Valid
              (bind _output_get (Variables|/Script/FortniteGame.FortMinigame|SetTotalRounds (Variables|UserOptions|GetTotalRounds) _minigame))
              (if (> _output_get 1)
                (GameplayTags|AddGameplayTag (Variables|/Script/FortniteGame.FNEUIComponent|GetUIExtensionTags _returnvalue) "(TagName=\"Minigame.UIExtension.Rounds\")" :!class /Script/GameplayTags.BlueprintGameplayTagLibrary)))))))))

(fn UpdateUISettings
  (bind _useroptioncomponent_cosmetic (Variables|Default|GetUserOptionComponent_Cosmetic))
  (Variables|/CreativeCoreDevices/ExperienceSettings/BP_UserOptionComponent_Cosmetic.BP_UserOptionComponent_Cosmetic_C|SetbUseLatestHUD (Variables|UserOptions|Scoped|GetbUseLatestHUD) _useroptioncomponent_cosmetic)
  (Class|BPUserOptionComponentCosmetic|HandleOptionsChanged _useroptioncomponent_cosmetic))

(fn UpdateVoiceChatExperienceSettings
  (bind (_voicechat _keepspectatorsinteamchat _attenuatevoice _fullvolumedistance _falloffdistance) (Class|GetClassDefaults "/CreativeCoreDevices/ExperienceSettings/ExperienceSettingsComponent_Minigame.ExperienceSettingsComponent_Minigame_C"))
  (ScalableFloat|SetValuewithContext _voicechat (Utilities|Select :Index (Variables|UserOptions|SharedSettings|GetVoiceChat) :NewEnumerator0 "None" :NewEnumerator1 "ScopeTeamOnly" :NewEnumerator2 "ScopeOnly"))
  (ScalableFloat|SetValuewithContext _keepspectatorsinteamchat (Variables|UserOptions|SharedSettings|GetKeepSpectatorsInTeamChat))
  (ScalableFloat|SetValuewithContext _attenuatevoice (Variables|UserOptions|SharedSettings|GetbIsProximityChatEnabled))
  (ScalableFloat|SetValuewithContext _fullvolumedistance (Variables|UserOptions|SharedSettings|GetProximityChatDistanceInMeters))
  (ScalableFloat|SetValuewithContext _falloffdistance (Variables|UserOptions|SharedSettings|GetAdditionalFalloffDistanceInMeters)))

(fn ConstructionScript)
