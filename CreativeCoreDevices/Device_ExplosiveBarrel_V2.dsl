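; Creative explosive barrel device that blows up on death, proximity overlap, timed fuse or external event while a minigame runs.
; Explosion queues throttled async sphere overlaps for actors and building structures, sends each hit a gameplay event to a damage ability, then hides the barrel.
; Minigame state changes reset health, timers, fuse and collision.
; User options drive async-loaded meshes, range visualization and health bar, with migration from legacy settings.
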(event MinigameStateChangedDelegate(FortMinigameLogic)
  (params (Minigame :type Fort%20Minigame%20Object%20Reference) (NewMinigameState :type EFortMinigameState%20Enum :default "PreGame"))
  (bind _returnvalue (Utilities|Enum|Equal(Enum) NewMinigameState "PostGameReset"))
  ;[0.25 0.25 0.25] Handle Visibility and collision when MinigameStateChanged
  (Utilities|FlowControl|Sequence
    (:then_0
      (CallFunction|UpdateVisualizationState)
      (CallFunction|SetWakeIgnoreState _returnvalue))
    (:then_1
      (if _returnvalue
        (CallFunction|SetCollisionForcedDuringGame "NewEnumerator0")
        ;[1 1 1] Cancel detonation timer when minigame ends
        (Utilities|Time|ClearandInvalidateTimerbyHandle (Variables|Default|GetDetonationTimer))
        (CallFunction|SetInvulneravilitybyMinigameState)
        (CallFunction|RestoretoMaxHealth)
        (CallFunction|UpdateBarrelMeshCollision)
        (Variables|Default|SetExplosiveFuseLit false)
        (Variables|Default|SetTriggerBarrelKiller 0)))))

(event MinigameEndedDelegate(FortMinigameLogic)
  ;[1 1 1] Cancel detonation timer when minigame ends
  (Utilities|Time|ClearandInvalidateTimerbyHandle (Variables|Default|GetDetonationTimer))
  (CallFunction|SetInvulneravilitybyMinigameState)
  (CallFunction|RestoretoMaxHealth)
  (CallFunction|UpdateBarrelMeshCollision)
  (Variables|Default|SetExplosiveFuseLit false)
  (Variables|Default|SetTriggerBarrelKiller 0))

(event Custom|Reset%20Barrel%20State
  (if (not (Utilities|Time|IsTimerActivebyHandle (Variables|Default|GetResetTimerHandle)))
    (Utilities|FlowControl|Sequence
      (:then_0
        (CallFunction|SetCollisionForcedDuringGame "NewEnumerator0")
        ;[1 1 1] Cancel detonation timer when minigame ends
        (Utilities|Time|ClearandInvalidateTimerbyHandle (Variables|Default|GetDetonationTimer))
        (CallFunction|SetInvulneravilitybyMinigameState)
        (CallFunction|RestoretoMaxHealth)
        (CallFunction|UpdateBarrelMeshCollision)
        (Variables|Default|SetExplosiveFuseLit false)
        (Variables|Default|SetTriggerBarrelKiller 0))
      (:then_1
        (CallFunction|SetVisibilityForcedDuringGame "NewEnumerator0")
        (Visibility|UpdateVisibility (Variables|Default|GetVisibleInGameComponent))
        (bind _returnvalue (Utilities|Time|SetTimerbyEvent (delegate Custom|On%20Reset%20Timer%20Expired) (Variables|UserOptions|GetTime%20Until%20Reset%20Allowed)))
        (Variables|Default|SetResetTimerHandle _returnvalue))
      (:then_2
        ;[1 1 1] Setup Detonation Timer
        (if (Variables|UserOptions|GetHas%20Timed%20Detonation%20From%20Game%20Start)
          (bind _returnvalue_2 (Utilities|Time|SetTimerbyEvent (delegate Custom|TimerDetonationEvent) (Variables|UserOptions|GetTime%20to%20Detonation%20From%20Game%20Start)))
          (Variables|Default|SetDetonationTimer _returnvalue_2)
          (Variables|Default|SetExplosiveFuseLit true))))))

(event Custom|On%20Reset%20Timer%20Expired)

(event MinigameStartedDelegate(FortMinigameLogic)
  ;[1 1 1] Timer detonation.  Only works when a minigame is started.
  (CallFunction|SetInvulneravilitybyMinigameState)
  ;[1 1 1] Setup Detonation Timer
  (if (Variables|UserOptions|GetHas%20Timed%20Detonation%20From%20Game%20Start)
    (bind _returnvalue (Utilities|Time|SetTimerbyEvent (delegate Custom|TimerDetonationEvent) (Variables|UserOptions|GetTime%20to%20Detonation%20From%20Game%20Start)))
    (Variables|Default|SetDetonationTimer _returnvalue)
    (Variables|Default|SetExplosiveFuseLit true)))

(event Custom|TimerDetonationEvent
  ;[1 1 1] Setup Detonation Timer
  (CallFunction|ExplodeBarrel))

(event EventBeginPlay
  ;[0 1 0.2776 0.25] Initialize settings
  (Visibility|SetVisibilityComponents (Variables|Default|GetVisibleInGameComponent) (Utilities|Array|MakeArray (Variables|Default|GetSM_MilitaryBarrel) (Variables|Default|GetBombSparkFX)) true)
  (CallFunction|SetInvulneravilitybyMinigameState)
  (Class|FortAsyncActionAuthorityMatchReady|MatchReadyAsync
    (:OnReady
      (CallFunction|SetMaxHealth)
      (Components|Sphere|SetSphereRadius (Variables|Default|GetProximityTrigger) (* (Variables|UserOptions|Basic|GetExplode%20on%20Proximity%20Range) 512.0) false) ; Sets size of prox trigger based on setting
      (CallFunction|UpdateHealthWidget)
      (CallFunction|UpdateBarrelMeshCollision)
      (CallFunction|UpdateVisibilityComponent)
      (Networking|SetNetDormancy "DORM_DormantAll"))) ; Ensures everything exists before getting the initial settings correct. \nWithout it the Set Max Health fails because the Ability system component doesn't exist when begin play fires.
)

(event Custom|Initialize%20Device
  ;[0 1 0.2776 0.25] Initialize settings
  (CallFunction|UpdateVisualMesh)
  (CallFunction|UpdateVisualizationState)
  (CallFunction|SetMaxHealth)
  (Components|Sphere|SetSphereRadius (Variables|Default|GetProximityTrigger) (* (Variables|UserOptions|Basic|GetExplode%20on%20Proximity%20Range) 512.0) false) ; Sets size of prox trigger based on setting
  (CallFunction|UpdateHealthWidget)
  (CallFunction|UpdateBarrelMeshCollision)
  (CallFunction|UpdateVisibilityComponent)
  (Networking|SetNetDormancy "DORM_DormantAll"))

(event EditorPreview|EventSetupForEditorPreview
  ;[0 1 0.2776 0.25] Initialize settings
  (CallFunction|InitializeDevice))

(event OnAnyPropertyChanged(ToyOptionsComponent)
  ;[0 1 0.2776 0.25] Initialize settings
  (CallFunction|InitializeDevice))

(event OnPropertiesLoaded(ToyOptionsComponent)
  ;[0 1 0.2776 0.25] Initialize settings
  (CallFunction|InitializeDevice))

(event Custom|PlayBarrelFX
  ;[1 1 1] FX when Barrel is killed with Damage
  (Networking|CanExecuteCosmeticEvents
    (:True
      (CallFunction|OnExplodeFX "0, 0, 0"))))

(event Custom|OnDelayedProximity
  ;[1 1 1] ProximityDelay
  (CallFunction|ExplodeBarrel))

(event OnComponentBeginOverlap(ProximityTrigger)
  (params (OverlappedComponent :type Primitive%20Component%20Object%20Reference) (OtherActor :type Actor%20Object%20Reference) (OtherComp :type Primitive%20Component%20Object%20Reference) (OtherBodyIndex :type Integer) (bFromSweep :type Boolean) (SweepResult :type Hit%20Result%20Structure%20%28by%20ref%29))
  ;[1 1 1] Proximity trigger explosion, only works during minigames
  (bind _isrunning (CallFunction|MinigameIsinProgress true))
  (if (and _isrunning (Variables|UserOptions|Basic|GetExplode%20on%20Proximity))
    (bind (_shouldtrigger _controller) (CallFunction|ShouldTriggerProximityExplosion OtherActor))
    (if _shouldtrigger
      (if (> (Variables|UserOptions|GetProximity%20Delay) 0.0)
        ;[1 1 1] ProximityDelay
        (if (not (Utilities|Time|IsTimerActivebyHandle (Variables|Default|GetDelayedProximityTH)))
          (Variables|Default|SetTriggerBarrelKiller _controller)
          (bind _returnvalue (Utilities|Time|SetTimerbyEvent (EventDispatchers|CreateEvent :!event Custom|OnDelayedProximity) (Variables|UserOptions|GetProximity%20Delay)))
          (Variables|Default|SetDelayedProximityTH _returnvalue)
          (Variables|Default|SetExplosiveFuseLit true))
        (else
          (CallFunction|ExplodeBarrel))))))

(event EventOnDeathPlayEffects
  (params (Damage :type Float%20%28single-precision%29) (DamageTags :type Gameplay%20Tag%20Container%20Structure%20%28by%20ref%29) (Momentum :type Vector :default "0, 0, 0") (HitInfo :type Hit%20Result%20Structure%20%28by%20ref%29) (InstigatedBy :type Fort%20Pawn%20Object%20Reference) (DamageCauser :type Actor%20Object%20Reference) (EffectContext :type Gameplay%20Effect%20Context%20Handle%20Structure))
  (bind _isrunning (CallFunction|MinigameIsinProgress false))
  (if _isrunning
    (Rendering|SetVisibility (Variables|Default|GetSM_MilitaryBarrel) false false)
    (Components|Activation|SetActive (Variables|Default|GetBombSparkFX) false false)))

(event EventOnDamageServer
  (params (Damage :type Float%20%28single-precision%29) (DamageTags :type Gameplay%20Tag%20Container%20Structure%20%28by%20ref%29) (Momentum :type Vector :default "0, 0, 0") (HitInfo :type Hit%20Result%20Structure%20%28by%20ref%29) (InstigatedBy :type Controller%20Object%20Reference) (DamageCauser :type Actor%20Object%20Reference) (EffectContext :type Gameplay%20Effect%20Context%20Handle%20Structure)))

(event Custom|Turn%20On%20Visibility%20Event
  (params (Instigator :type Controller%20Object%20Reference))
  ;[0.2393 0 1 0.25] Gameplay Functions
  (CallFunction|SetVisibilityForcedDuringGame "NewEnumerator1"))

(event Custom|Explode%20Barrel
  ;[1 1 1] Blow up barrel
  (bind _isrunning (CallFunction|MinigameIsinProgress false))
  (if (and _isrunning (not (Variables|Replicated|GetHasExploded)))
    (Utilities|FlowControl|Sequence
      (:then_0
        (Variables|Default|SetBarrelKiller (stale EventOnDeathServer InstigatedBy))
        (Curie|ResetCurieActor self))
      (:then_1
        (bind _returnvalue (Utility|OverlapQueue|GetAsyncOverlapQueue self))
        (Class|AsyncOverlapQueue|ThrottledAsyncSphereOverlapActors :self _returnvalue :Location (Transformation|GetWorldLocation (Variables|Default|GetProximityTrigger)) :Radius (* (Variables|UserOptions|Basic|GetBlast%20Radius) 512.0) :bCheckLineOfSight (Variables|UserOptions|GetBlastLineOfSight) :ShouldCancelProcess (EventDispatchers|CreateEvent :!event ShouldCancelExplosion) :ProcessHitActorFunction (delegate Custom|DestroyHitActors) :CompletionFunction (delegate Custom|FinishedDamagingActors) :ObjectTypes "ObjectTypeQuery1" :TraceChannel "ECC_Visibility")
        (bind _returnvalue_2 (Class|PlayspaceComponentExplosiveDeviceAsyncOverlap|GetExplosiveDeviceAsyncOverlapQueue))
        (Class|ExplosiveDeviceAsyncOverlapQueue|ThrottledAsyncSphereOverlapOverlappingBuildingActors _returnvalue_2 (Transformation|GetWorldLocation (Variables|Default|GetProximityTrigger)) (* (Variables|UserOptions|Basic|GetBlast%20Radius) 512.0) (* (+ (Variables|UserOptions|Basic|GetBlast%20Radius) 0.5) 512.0) (Variables|UserOptions|GetBlastLineOfSight) (EventDispatchers|CreateEvent :!event ShouldCancelExplosion) (delegate Custom|ProcessHitOverlappingStructure) (delegate Custom|OverlappingStructureOverlapComplete)))
      (:then_2
        (if (Variables|UserOptions|GetPlay%20Audio%20/%20VFX)
          (CallFunction|PlayBarrelFX)))
      (:then_3
        (Variables|Replicated|SetHasExploded true)
        (GameplayEvent|BroadcastEvent (Variables|UserOptions|Events|GetOn%20Exploded) 0 (Utilities|Struct|MakeGameplayEventDestroyed (select (Utilities|IsValid (Variables|Default|GetTriggerBarrelKiller) :!class /Script/Engine.KismetSystemLibrary) (Variables|Default|GetTriggerBarrelKiller) (Variables|Default|GetBarrelKiller))))
        (CallFunction|SetVisibilityForcedDuringGame "NewEnumerator2")
        (CallFunction|SetCollisionForcedDuringGame "NewEnumerator2")
        (Networking|FlushNetDormancy)
        (Utilities|Time|ClearandInvalidateTimerbyHandle (Variables|Default|GetDelayedProximityTH))
        (Variables|Default|SetExplosiveFuseLit false)))))

(event EventOnDeathServer
  (params (Damage :type Float%20%28single-precision%29) (DamageTags :type Gameplay%20Tag%20Container%20Structure%20%28by%20ref%29) (Momentum :type Vector :default "0, 0, 0") (HitInfo :type Hit%20Result%20Structure%20%28by%20ref%29) (InstigatedBy :type Controller%20Object%20Reference) (DamageCauser :type Actor%20Object%20Reference) (EffectContext :type Gameplay%20Effect%20Context%20Handle%20Structure))
  ;[1 1 1] Blow up barrel
  (bind _isrunning (CallFunction|MinigameIsinProgress false))
  (if (and _isrunning (not (Variables|Replicated|GetHasExploded)))
    (Utilities|FlowControl|Sequence
      (:then_0
        (Variables|Default|SetBarrelKiller InstigatedBy)
        (Curie|ResetCurieActor self))
      (:then_1
        (bind _returnvalue (Utility|OverlapQueue|GetAsyncOverlapQueue self))
        (Class|AsyncOverlapQueue|ThrottledAsyncSphereOverlapActors :self _returnvalue :Location (Transformation|GetWorldLocation (Variables|Default|GetProximityTrigger)) :Radius (* (Variables|UserOptions|Basic|GetBlast%20Radius) 512.0) :bCheckLineOfSight (Variables|UserOptions|GetBlastLineOfSight) :ShouldCancelProcess (EventDispatchers|CreateEvent :!event ShouldCancelExplosion) :ProcessHitActorFunction (delegate Custom|DestroyHitActors) :CompletionFunction (delegate Custom|FinishedDamagingActors) :ObjectTypes "ObjectTypeQuery1" :TraceChannel "ECC_Visibility")
        (bind _returnvalue_2 (Class|PlayspaceComponentExplosiveDeviceAsyncOverlap|GetExplosiveDeviceAsyncOverlapQueue))
        (Class|ExplosiveDeviceAsyncOverlapQueue|ThrottledAsyncSphereOverlapOverlappingBuildingActors _returnvalue_2 (Transformation|GetWorldLocation (Variables|Default|GetProximityTrigger)) (* (Variables|UserOptions|Basic|GetBlast%20Radius) 512.0) (* (+ (Variables|UserOptions|Basic|GetBlast%20Radius) 0.5) 512.0) (Variables|UserOptions|GetBlastLineOfSight) (EventDispatchers|CreateEvent :!event ShouldCancelExplosion) (delegate Custom|ProcessHitOverlappingStructure) (delegate Custom|OverlappingStructureOverlapComplete)))
      (:then_2
        (if (Variables|UserOptions|GetPlay%20Audio%20/%20VFX)
          (CallFunction|PlayBarrelFX)))
      (:then_3
        (Variables|Replicated|SetHasExploded true)
        (GameplayEvent|BroadcastEvent (Variables|UserOptions|Events|GetOn%20Exploded) 0 (Utilities|Struct|MakeGameplayEventDestroyed (select (Utilities|IsValid (Variables|Default|GetTriggerBarrelKiller) :!class /Script/Engine.KismetSystemLibrary) (Variables|Default|GetTriggerBarrelKiller) (Variables|Default|GetBarrelKiller))))
        (CallFunction|SetVisibilityForcedDuringGame "NewEnumerator2")
        (CallFunction|SetCollisionForcedDuringGame "NewEnumerator2")
        (Networking|FlushNetDormancy)
        (Utilities|Time|ClearandInvalidateTimerbyHandle (Variables|Default|GetDelayedProximityTH))
        (Variables|Default|SetExplosiveFuseLit false)))))

(event Custom|Turn%20Off%20Visibility%20Event
  (params (Instigator :type Controller%20Object%20Reference))
  ;[0.2393 0 1 0.25] Gameplay Functions
  (CallFunction|SetVisibilityForcedDuringGame "NewEnumerator2"))

(event Custom|Explode%20Event
  (params (Instigator :type Controller%20Object%20Reference))
  ;[0.2393 0 1 0.25] Gameplay Functions
  (if (not (Variables|Replicated|GetHasExploded))
    (bind _isrunning (CallFunction|MinigameIsinProgress false)) ; only explode if the game is currently running.
    (if _isrunning
      (Variables|Default|SetTriggerBarrelKiller Instigator)
      (CallFunction|ExplodeBarrel))))

(event Custom|DestroyHitActors
  (params (HitActor :type Actor%20Object%20Reference))
  ;[1 1 1] Blow up barrel
  (CallFunction|BlowUpBarrel HitActor))

(event Custom|ProcessHitOverlappingStructure
  (params (HitActor :type Actor%20Object%20Reference))
  ;[1 1 1] Blow up barrel
  (CallFunction|BlowUpBarrel HitActor))

(event Custom|FinishedDamagingActors
  (params (WasCancelled :type Boolean)))

(event Custom|Reset%20Event
  (params (Instigator :type Controller%20Object%20Reference))
  ;[0.2393 0 1 0.25] Gameplay Functions
  (CallFunction|ResetBarrelState))

(event Custom|OverlappingStructureOverlapComplete
  (params (WasCancelled :type Boolean)))

(event OnVisibilityChanged(VisibleInGameComponent)
  (params (Visible :type Boolean))
  (CallFunction|UpdateBarrelMeshCollision))

(event Custom|UpdateVisualMesh
  (bind _bombsparkfx (Variables|Default|GetBombSparkFX))
  (bind _sm_militarybarrel (Variables|Default|GetSM_MilitaryBarrel))
  ;[1 1 1] Set the Visual Mesh of the Device
  (bind _loaded_object (Creative|Loading|AsyncLoadEditorSafe (Variables|UserOptions|Advanced|GetDeviceMesh))
    (:Loaded
      (bind _asstatic_mesh (Utilities|Casting|CastToStaticMesh _loaded_object))
      (Components|StaticMesh|SetStaticMesh _sm_militarybarrel _asstatic_mesh)
      (Rendering|Material|SetMaterial _sm_militarybarrel 0 (StaticMesh|GetMaterial _asstatic_mesh 0))
      (if (== (Variables|UserOptions|Advanced|GetDeviceMesh) (Variables|Default|GetBombMeshReference))
        (bind _loaded_object_2 (Creative|Loading|AsyncLoadEditorSafe (Variables|Default|GetFuseFX))
          (:Loaded
            (bind _asniagara_system (Utilities|Casting|CastToNiagaraSystem _loaded_object_2))
            (Niagara|SetNiagaraSystemAsset _bombsparkfx _asniagara_system true)))
        (else
          (Niagara|SetNiagaraSystemAsset _bombsparkfx 0 true))))))

(event Custom|ApplyExplosionImpulseOnClientPawn
  (params (Target :type Fort%20Pawn%20Object%20Reference) (LaunchVelocity :type Vector))
  ;[1 1 1] Called from the explosion GA - used with physics until LaunchCharacter can be called from server
  (if (Utilities|Enum|Equal(Enum) (Networking|GetLocalRole Target) "ROLE_AutonomousProxy")
    (Pawn|Character|LaunchCharacterJump Target LaunchVelocity false false false false)))

(event Custom|SetCollisionForcedDuringGame
  (params (NewForcedValue :type Enum_ForcedValueDuringGame%20Enum :default "NewEnumerator0"))
  ;[0.25 0.1274 0.02404] ForcedValuesDuringGame
  (if (Utilities|Enum|NotEqual(Enum) (Variables|Replicated|GetCollisionForcedDuringGame) NewForcedValue)
    (Variables|Replicated|SetCollisionForcedDuringGame NewForcedValue)
    (Networking|FlushNetDormancy)))

(event Custom|Async%20Load%20Visualization%20Material
  (bind _range_visualization (Variables|UserOptions|GetRange%20Visualization))
  (bind _rangevisualizationmaterials (Variables|Default|GetRangeVisualizationMaterials))
  (bind (_value _returnvalue) (Utilities|Map|Find _rangevisualizationmaterials _range_visualization))
  ;[1 0.7441 0 0.5] Load the Material asyncronously, then show the range visualization if the pawn is currently overlapping trigger
  (Variables|Default|SetMaterial%20Loaded false)
  (if _returnvalue
    (bind _loaded_object (Creative|Loading|AsyncLoadEditorSafe _value)
      (:Loaded
        (bind _asmaterial_interface (Utilities|Casting|CastToMaterialInterface _loaded_object))
        (Rendering|Material|SetMaterial (Variables|Default|GetRangeSphere) 0 _asmaterial_interface)
        (Variables|Default|SetMaterial%20Loaded true)
        (if (Editor|InEditor)
          (CallFunction|ShowRangeVisualization)
          (else
            ;;[1 0.4717 0] If we're currently overlapping a local pawn, it won't have started the range visualization as the material wasn't loaded. Start it now.
            (bind _array_element (Utilities|Array|ForEachLoopwithBreak (Fortnite|GetLocalFortPlayerControllers true) :!label _label)
              (:LoopBody
                (Utilities|IsValid (Pawn|GetControlledPawn _array_element)
                  (:Is%20Valid
                    (if (Collision|IsOverlappingActor (Variables|Default|GetRange%20Visualization%20Trigger) (Pawn|GetControlledPawn _array_element) :!class /Script/Engine.PrimitiveComponent)
                      (CallFunction|ShowRangeVisualization)
                      (goto _label Break))))))))))))

(event OnComponentBeginOverlap(RangeVisualizationTrigger)
  (params (OverlappedComponent :type Primitive%20Component%20Object%20Reference) (OtherActor :type Actor%20Object%20Reference) (OtherComp :type Primitive%20Component%20Object%20Reference) (OtherBodyIndex :type Integer) (bFromSweep :type Boolean) (SweepResult :type Hit%20Result%20Structure%20%28by%20ref%29))
  ;[1 0.8995 0 0.25] Show Visualization of Range when Near
  (if (Variables|Default|GetMaterial%20Loaded)
    (CallFunction|ShowRangeVisualization)))

(event OnComponentEndOverlap(RangeVisualizationTrigger)
  (params (OverlappedComponent :type Primitive%20Component%20Object%20Reference) (OtherActor :type Actor%20Object%20Reference) (OtherComp :type Primitive%20Component%20Object%20Reference) (OtherBodyIndex :type Integer))
  ;[1 0.8995 0 0.25] Show Visualization of Range when Near
  (CallFunction|HideRangeVisualization))

(event Custom|ShowRangeVisualization
  ;[1 0.8995 0 0.25] Show Visualization of Range when Near
  (Networking|CanExecuteCosmeticEvents
    (:True
      (Rendering|SetVisibility (Variables|Default|GetRangeSphere) true false)
      (if (Editor|InEditor)
        (Transformation|SetWorldScale3D (Variables|Default|GetRangeSphere) (CallFunction|GetRadiusVisualizationRange))
        (else
          (bind (_direction _lerp) (Timeline|Visualization%20Range%20%25 Play)
            (:Update
              (Rendering|Material|SetScalarParameterValue (Material|GetMID (Variables|Default|GetRangeSphere)) "Opacity" _lerp :!class /Script/Engine.MaterialInstanceDynamic)
              (Transformation|SetWorldScale3D (Variables|Default|GetRangeSphere) (Math|Vector|Lerp(Vector) "0.010000,0.010000,0.010000" (CallFunction|GetRadiusVisualizationRange) _lerp)))
            (:Finished
              (switch Utilities|FlowControl|Switch|SwitchonETimelineDirection _direction
                (:Backward
                  (Rendering|SetVisibility (Variables|Default|GetRangeSphere) false false))))))))))

(event Custom|HideRangeVisualization
  ;[1 0.8995 0 0.25] Show Visualization of Range when Near
  (Networking|CanExecuteCosmeticEvents
    (:True
      (if (not (Editor|InEditor))
        (Timeline|Visualization%20Range%20%25 Reverse)
        (else
          (Transformation|SetWorldScale3D (Variables|Default|GetRangeSphere) "0.010000,0.010000,0.010000")
          (Rendering|SetVisibility (Variables|Default|GetRangeSphere) false false))))))

(event Custom|ToggleFuseFX
  (params (Activate :type Boolean))
  ;[1 0.8995 0 0.25] Toggle Fuse VFX
  (if (== (Variables|UserOptions|Advanced|GetDeviceMesh) (Variables|Default|GetBombMeshReference))
    (Components|Activation|SetActive (Variables|Default|GetBombSparkFX) Activate false)
    (bind _loaded_object (Creative|Loading|AsyncLoadEditorSafe (select Activate "/Game/Creative/Devices/Bomb/Materials/CP_MI_Device_Bomb_Lit_01.CP_MI_Device_Bomb_Lit_01" "/Game/Creative/Devices/Bomb/Materials/CP_MI_Device_Bomb_01.CP_MI_Device_Bomb_01"))
      (:Loaded
        (bind _asmaterial_instance (Utilities|Casting|CastToMaterialInstance _loaded_object))
        (Rendering|Material|SetMaterial (Variables|Default|GetSM_MilitaryBarrel) 0 _asmaterial_instance)))))

(fn BlowUpBarrel
  (params (ActorToHit :type Actor%20Object%20Reference))
  ;[1 1 1] Send Gameplay event to ability to get explosion to fire and damage specified actor via OptionalObject
  (Ability|SendGameplayEventtoActor self (Variables|Default|GetEventExplosionTag) (Utilities|Struct|MakeGameplayEventData (Variables|Default|GetEventExplosionTag) (select (Utilities|IsValid (Variables|Default|GetTriggerBarrelKiller) :!class /Script/Engine.KismetSystemLibrary) (Variables|Default|GetTriggerBarrelKiller) (Variables|Default|GetBarrelKiller)) 0 ActorToHit 0 "()" "(GameplayTags=)" "(GameplayTags=)" (Variables|UserOptions|Basic|GetPlayer%20Damage))))

(fn GetRadiusVisualizationRange
  (return (Math|Conversions|ToVector(Float) (Utilities|Select :Index (Variables|UserOptions|GetRange%20Visualization) :NewEnumerator0 0.0 :NewEnumerator1 (* (Variables|UserOptions|Basic|GetBlast%20Radius) 10.24) :NewEnumerator2 (* (Variables|UserOptions|Basic|GetExplode%20on%20Proximity%20Range) 10.24)))))

(fn GetCollisionStaticMesh
  (return "0, 0, 0" (Variables|/Script/Engine.StaticMeshComponent|GetStaticMesh (Variables|Default|GetSM_MilitaryBarrel))))

(fn GetFriendlyTeamArray
  (return (select (Creative|TeamOption|IsTeamType (Variables|UserOptions|GetIgnore%20Team%20For%20Damage)) (Utilities|Array|MakeArray (- (Creative|TeamOption|GetLegacyTeamIndex (Variables|UserOptions|GetIgnore%20Team%20For%20Damage) false) 1)) 0)))

(fn GetHostileTeamArray
  (bind _ignore_team_for_proximity (Variables|UserOptions|GetIgnore%20Team%20For%20Proximity))
  (Utilities|FlowControl|Sequence
    (:then_0
      (LocalVariables|SetHostile%20Team%20Array (Knobs|Team|GetTeamArray self))
      (if (Variables|UserOptions|Basic|GetExplode%20on%20Proximity)
        (if (Creative|TeamOption|IsTeamType _ignore_team_for_proximity)
          (Utilities|Array|RemoveItem (LocalVariables|GetHostile%20Team%20Array) (- (Creative|TeamOption|GetLegacyTeamIndex _ignore_team_for_proximity false) 1)) ; Removes the team from the hostile team array if they're not going to proximity activate it.
        )))
    (:then_1
      (return (LocalVariables|GetHostile%20Team%20Array)))))

(fn GetMeshComponents
  (return 0))

(fn IgnoreWorldGridScaleOverride
  (return false))

(fn IsInPlayMode
  (bind _fortminigamelogic (Variables|Default|GetFortMinigameLogic)
    (:then
      (bind _currentminigame (Variables|/Script/FortniteGame.FortMinigameLogicComponent|GetCurrentMinigame _fortminigamelogic)
        (:then
          (return (Minigame|IsinPlayMode _currentminigame :!class /Script/FortniteGame.FortMinigame)))
        (:else
          _)))
    (:else
      _))
  (return false))

(fn IsTargetableByLinearMover
  (return true))

(fn LocalPlayerShouldSeeBarrel
  (Networking|SwitchHasAuthority
    (:Authority
      (return true))
    (:Remote
      (Utilities|FlowControl|Sequence
        (:then_0
          (switch Utilities|FlowControl|Switch|SwitchonECreativeTeamType (Utilities|Struct|BreakCreativeTeamOption (Variables|UserOptions|GetTeam%20Visibility))
            (:TeamIndex
              ;[0.25 0.5 1] Team Index
              (return (== (Creative|TeamOption|GetLegacyTeamIndex (Variables|UserOptions|GetTeam%20Visibility) false) (Math|Conversions|ToInteger(Byte) (Team|GetActorPvPTeamIndex (Game|GetPlayerController 0))))))
            (:Any
              ;[0.25 0.5 1] Any
              (return true))
            (:Hostiles
              ;[0.25 0.5 1] Hostiles
              (return (Utilities|Enum|Equal(Enum) (Team|GetTeamAffiliation self (Game|GetPlayerPawn 0)) "Hostile")))
            (:Friendlies
              ;[0.25 0.5 1] Friendlies
              (return (Utilities|Enum|Equal(Enum) (Team|GetTeamAffiliation self (Game|GetPlayerPawn 0)) "Friendly")))
            (:Neutral
              ;[0.25 0.5 1] Neutral
              (return (Utilities|Enum|Equal(Enum) (Team|GetTeamAffiliation self (Game|GetPlayerPawn 0)) "Neutral")))))
        (:then_1
          ;[0.25 0.5 1] All Other Types
          (return false))))))

(fn MigrateLegacyUserOptions
  (bind _timerdetonatesetting (Class|GetClassDefaults "/CreativeCoreDevices/Device_ExplosiveBarrel_V2.Device_ExplosiveBarrel_V2_C"))
  (bind _damagerangesetting (Class|GetClassDefaults "/CreativeCoreDevices/Device_ExplosiveBarrel_V2.Device_ExplosiveBarrel_V2_C"))
  (bind _proximitydistancesetting (Class|GetClassDefaults "/CreativeCoreDevices/Device_ExplosiveBarrel_V2.Device_ExplosiveBarrel_V2_C"))
  (bind _startinghealthsetting (Class|GetClassDefaults "/CreativeCoreDevices/Device_ExplosiveBarrel_V2.Device_ExplosiveBarrel_V2_C"))
  (bind _startinghealthsetting_2 (Variables|Legacy|GetStartingHealthSetting))
  (bind _timerdetonatesetting_2 (Variables|Legacy|GetTimerDetonateSetting))
  (bind _proximitydistancesetting_2 (Variables|Legacy|GetProximityDistanceSetting))
  (bind _damagerangesetting_2 (Variables|Legacy|GetDamageRangeSetting))
  (|CallParentFunction)
  (Utilities|FlowControl|Sequence
    (:then_0
      (Utilities|FlowControl|Sequence
        (:then_0
          ;[0 1 0.4839 0.25] Migrate old Damage Range Setting to Blast Radius
          (if (!= _damagerangesetting _damagerangesetting_2)
            (Variables|UserOptions|Basic|SetBlast%20Radius (/ _damagerangesetting_2 512.0))
            (Variables|Legacy|SetDamageRangeSetting _damagerangesetting)
            (LocalVariables|SetOptions%20Migrated true)))
        (:then_1
          ;[0 1 0.8219 0.25] Migrate Starting Health Setting to Can Be Damaged or Health
          (if (!= _startinghealthsetting _startinghealthsetting_2)
            (if (< _startinghealthsetting_2 0)
              (Variables|UserOptions|Basic|SetCan%20be%20Damaged false)
              (else
                (Variables|UserOptions|Basic|SetHealth _startinghealthsetting_2)))
            (Variables|Legacy|SetStartingHealthSetting _startinghealthsetting)
            (LocalVariables|SetOptions%20Migrated true)))
        (:then_2
          ;[0 0.9574 1 0.25] Migrate Timer Detonate Setting to Has Timed Detotnation and Time To Detonation
          (if (!= _timerdetonatesetting _timerdetonatesetting_2)
            (Variables|UserOptions|SetHas%20Timed%20Detonation%20From%20Game%20Start true)
            (Variables|UserOptions|SetTime%20to%20Detonation%20From%20Game%20Start _timerdetonatesetting_2)
            (Variables|Legacy|SetTimerDetonateSetting _timerdetonatesetting)
            (LocalVariables|SetOptions%20Migrated true)))
        (:then_3
          ;[0 0.7571 1 0.25] Migrate Proximity Distance Setting to Explode on Proximity and Explode On Proximity Range
          (if (!= _proximitydistancesetting _proximitydistancesetting_2)
            (Variables|UserOptions|Basic|SetExplode%20on%20Proximity true)
            (Variables|UserOptions|Basic|SetExplode%20on%20Proximity%20Range (/ (Math|Conversions|ToFloat(Integer) _proximitydistancesetting_2) 512.0))
            (Variables|Legacy|SetProximityDistanceSetting _proximitydistancesetting)
            (LocalVariables|SetOptions%20Migrated true)))))
    (:then_1
      (if (LocalVariables|GetOptions%20Migrated)
        (OptionData|RefreshPropertyOverrides (Variables|Default|GetToyOptionsComponent))))))

(fn MinigameIsInProgress
  (params (IgnoreWarmUp :type Boolean))
  (bind _self self)
  (bind _returnvalue (Minigame|GetMinigame _self :!class /Script/FortniteGame.MinigameBlueprintLibrary))
  (Utilities|IsValid _returnvalue
    (:Is%20Valid
      ;[1 1 1] Is in progress pure function includes Warmup time, optionally ignore that and only use "In Progress" time
      (if IgnoreWarmUp
        (return (Utilities|Enum|Equal(Enum) (Variables|/Script/FortniteGame.FortMinigame|GetCurrentState _returnvalue) "InProgress"))
        (else
          (return (Minigame|IsinProgress _returnvalue)))))
    (:Is%20Not%20Valid
      (return false))))

(fn OnExplodeFX
  (params (Location :type Vector))
  (bind _returnvalue (Transformation|GetActorLocation))
  (bind _explosionsfx (LocalVariables|GetExplosionSFX))
  (if (Variables|UserOptions|GetPlay%20Audio)
    (Audio|PlaySoundatLocation _explosionsfx _returnvalue)
    (SoundIndicator|ShowSoundIndicatorforDuration (Variables|Default|GetFortSoundIndicator) _returnvalue _explosionsfx 1.0 0.3 0))
  (Effects|SpawnEmitteratLocation :EmitterTemplate "/Game/Creative/Devices/ExplosiveBarrel/P_ExplosiveBarrel_ExplosionMain.P_ExplosiveBarrel_ExplosionMain" :Location _returnvalue :Scale (Math|Conversions|ToVector(Float) (Variables|UserOptions|Basic|GetBlast%20Radius)) :bAutoDestroy true :PoolingMethod "None" :bAutoActivateSystem true))

(fn OnProximityTriggerableTeam
  (params (Actor :type Actor%20Object%20Reference))
  ;[1 1 1] Check if actor is on a team that should activate proximity trigger
  (return (or (Creative|TeamOption|IsTeamType (Variables|UserOptions|GetIgnore%20Team%20For%20Proximity) "None") (!= (Team|GetActorTeam Actor) (Math|Byte|MakeLiteralByte (Creative|TeamOption|ConverttoTeam (Variables|UserOptions|GetIgnore%20Team%20For%20Proximity)))))))

(fn OnRep_CollisionForcedDuringGame
  (CallFunction|UpdateBarrelMeshCollision))

(fn OnRep_ExplosiveFuseLit
  (Networking|CanExecuteCosmeticEvents
    (:True
      (CallFunction|ToggleFuseFX (Variables|Default|GetExplosiveFuseLit)))))

(fn RestoreToMaxHealth
  ;[1 0.4407 0.1563] Restores the building health to it's initial value
  (Networking|SwitchHasAuthority
    (:Authority
      (Health|SetHealth (Health|GetMaxHealth))
      (Variables|Replicated|SetHasExploded false)
      (CallFunction|SetMaxHealth)
      (Networking|FlushNetDormancy))))

(fn SetInvulneravilityByMinigameState
  (Networking|SwitchHasAuthority
    (:Authority
      (bind _isrunning (CallFunction|MinigameIsinProgress true))
      (Fortnite|SetCanbeDamaged self (and _isrunning (Variables|UserOptions|Basic|GetCan%20be%20Damaged))))))

(fn SetMaxHealth
  (bind _self self)
  (bind _returnvalue (Ability|GetAbilitySystemComponent _self))
  (bind _self_2 self)
  (bind _returnvalue_2 (Ability|GetAbilitySystemComponent _self_2))
  ;[1 1 1] Runs a gameplay effect on the actor that multiplies the max health by the setting value.
  (Networking|SwitchHasAuthority
    (:Authority
      (Utilities|IsValid _returnvalue_2
        (:Is%20Valid
          (bind _returnvalue_3 (Ability|GameplayEffect|AssignTagSetbyCallerMagnitude (GameplayEffects|MakeOutgoingSpec :self _returnvalue_2 :GameplayEffectClass (Variables|Default|GetModifyMaxHealth) :Level 0.0) (LocalVariables|GetSetMaxHealthDataTag) (Utilities|GetStartingHealth)))
          (Utilities|IsValid _returnvalue
            (:Is%20Valid
              (GameplayEffects|RemoveActiveGameplayEffect _returnvalue (Variables|Default|GetMaxHealthEffectHandler) -1)
              (bind _returnvalue_4 (GameplayEffects|ApplyGameplayEffectSpecToSelf _returnvalue _returnvalue_3))
              (Variables|Default|SetMaxHealthEffectHandler _returnvalue_4))))))))

(fn SetVisibilityForcedDuringGame
  (params (NewVisibilityForcedDuringGame :type Enum_ForcedValueDuringGame%20Enum :default "NewEnumerator0"))
  (Networking|SwitchHasAuthority
    (:Authority
      (if (not (and (Variables|Replicated|GetHasExploded) (Utilities|Enum|Equal(Enum) (Variables|GetNewVisibilityForcedDuringGame) "NewEnumerator1")))
        (if (Utilities|Enum|Equal(Enum) (Variables|UserOptions|GetCollision%20During%20Games) "NewEnumerator3")
          (CallFunction|SetCollisionForcedDuringGame (Variables|GetNewVisibilityForcedDuringGame)))
        (if (Utilities|Enum|Equal(Enum) (Variables|GetNewVisibilityForcedDuringGame) "NewEnumerator1")
          (Visibility|SetForcedVisibilityDuringGame (Variables|Default|GetVisibleInGameComponent) (select (CallFunction|LocalPlayerShouldSeeBarrel) "NewEnumerator1" "NewEnumerator2"))
          (else
            (Visibility|SetForcedVisibilityDuringGame (Variables|Default|GetVisibleInGameComponent) (Variables|GetNewVisibilityForcedDuringGame))))))))

(fn ShouldCancelExplosion
  (bind _isrunning (CallFunction|MinigameIsinProgress false))
  (return (not _isrunning)))

(fn ShouldDisplayDamageNumbers
  (return (Variables|UserOptions|GetDisplay%20Damage%20Numbers)))

(fn ShouldGenerateCurieContainer_BP
  (bind _isrunning (CallFunction|MinigameIsinProgress false))
  (return (and _isrunning (not (Variables|Replicated|GetHasExploded)))))

(fn ShouldInvisiblePropsBeAllowed
  (return false))

(fn ShouldTriggerProximityExplosion
  (params (OverlappingActor :type Object%20Reference) (OverlappingComponent :type Primitive%20Component%20Object%20Reference))
  ;[1 1 1] Ensures that overlapping actor should trigger the barrel.  Currently handles players and vehicles.
  (bind _asfort_player_pawn_athena (Utilities|Casting|CastToFortPlayerPawnAthena OverlappingActor)
    (:then
      (bind _returnvalue (CallFunction|OnProximityTriggerableTeam _asfort_player_pawn_athena))
      (return (and _returnvalue (not (Vehicle|IsinVehicle _asfort_player_pawn_athena))) (CreativeAI|Controller|GetParticipantControllerfromActor _asfort_player_pawn_athena false)))
    (:CastFailed
      (bind _asfort_vehicle_interface (Utilities|Casting|CastToFortVehicleInterface OverlappingActor)
        (:then
          (LocalVariables|SetlocalVehiclePassengers (Vehicle|GetAllPassengers _asfort_vehicle_interface))
          (if (Vehicle|HasDriver _asfort_vehicle_interface)
            (bind _returnvalue_2 (CallFunction|OnProximityTriggerableTeam (Vehicle|GetDriver _asfort_vehicle_interface)))
            (if _returnvalue_2
              (return true (CreativeAI|Controller|GetParticipantControllerfromActor (Vehicle|GetDriver _asfort_vehicle_interface) false))
              (else
                _)))
          (if (> (Utilities|Array|Length (LocalVariables|GetlocalVehiclePassengers)) 0)
            ;;[1 1 1] Checks if team of each passenger should trigger the barrel.  Breaks as soon as it finds one that should
            (for _array_element (LocalVariables|GetlocalVehiclePassengers)
              (bind _returnvalue_3 (CallFunction|OnProximityTriggerableTeam _array_element))
              (if _returnvalue_3
                (return true (CreativeAI|Controller|GetParticipantControllerfromActor _array_element false))))
            (return false 0)
            (else
              (return false 0))))
        (:CastFailed
          (return false 0))))))

(fn UpdateHealthWidget
  (Networking|CanExecuteCosmeticEvents
    (:True
      (if (not (Editor|InEditor))
        (if (Math|Boolean|ANDBoolean (Variables|UserOptions|GetShow%20Health%20Bar) (Utilities|Select :Index (Variables|UserOptions|GetHealth%20Bar%20Style) :NewEnumerator0 false :NewEnumerator1 true :NewEnumerator2 true) (Variables|UserOptions|Basic|GetCan%20be%20Damaged) (or (Utilities|Name|NotEqual(Name) (Collision|GetCollisionProfileName (Variables|Default|GetSM_MilitaryBarrel)) (CreativeDevices|GetNamefromCollisionProfile "(Name=\"FortCreativeOnlyVolumeManager\")")) (Variables|/Script/Engine.SceneComponent|GetbVisible (Variables|Default|GetSM_MilitaryBarrel))))
          (CallFunction|SetSuppressHealthBar true)
          (Class|CreativeValueBarComponent|SpawnOrUpdateValueBar (Variables|Default|GetFloatingHealthBar) (Variables|Default|GetSM_MilitaryBarrel) self "/Game/Creative/Devices/Common/Components/Creative_FloatingHealthBar.Creative_FloatingHealthBar_C" 0 (* (Variables|UserOptions|GetHide%20HUD%20Icon%20at) 1e+02) 0 (Utilities|GetStartingHealth) 2e+01 (Variables|UserOptions|GetRequires%20Line%20of%20Sight) (CallFunction|GetFriendlyTeamArray) (CallFunction|GetHostileTeamArray) (Creative|TeamOption|GetLegacyTeamIndex (Variables|UserOptions|GetTeam%20Visibility)) (Utilities|Select :Index (Variables|UserOptions|GetHealth%20Bar%20Style) :NewEnumerator0 false :NewEnumerator1 false :NewEnumerator2 true) (Utilities|String|ToString(Text) (Variables|UserOptions|GetFriendly%20Icon%20Text)) (Utilities|String|ToString(Text) (Variables|UserOptions|GetNeutral%20Icon%20Text)) (Utilities|String|ToString(Text) (Variables|UserOptions|GetHostile%20Icon%20Text)) (Variables|UserOptions|GetHUD%20Text%20Size) true (Variables|UserOptions|GetDisplayDistanceText) (Variables|UserOptions|GetClampToScreen) (Variables|UserOptions|GetShowOffscreenArrow) (Creative|UI|IconPicker|GetIconLibraryTexture (Variables|UserOptions|GetIcon%20Identifier)) true "(X=1.000000,Y=1.000000)" false false)
          (else
            (CallFunction|SetSuppressHealthBar (not (or (Variables|UserOptions|GetShow%20Health%20Bar) (not (Variables|UserOptions|Basic|GetCan%20be%20Damaged)))))
            (Class|CreativeValueBarComponent|HideValueBar (Variables|Default|GetFloatingHealthBar))))))
    (:False
      (bind _floatinghealthbar (Variables|Default|GetFloatingHealthBar)
        (:then
          (Components|DestroyComponent _floatinghealthbar))
        (:else)))))

(fn UpdateVisualizationState
  (Networking|CanExecuteCosmeticEvents
    (:True
      (if (and (not (Class|CreativeCommonDeviceFunctionLibrary|IsinPlayMode self)) (Utilities|Enum|NotEqual(Enum) (Variables|UserOptions|GetRange%20Visualization) "NewEnumerator0"))
        (Collision|SetCollisionEnabled (Variables|Default|GetRange%20Visualization%20Trigger) "QueryOnly")
        (Transformation|SetWorldScale3D (Variables|Default|GetRangeSphere) (CallFunction|GetRadiusVisualizationRange))
        (CallFunction|AsyncLoadVisualizationMaterial)
        (else
          _)))
    (:False
      _))
  (Collision|SetCollisionEnabled (Variables|Default|GetRange%20Visualization%20Trigger) "NoCollision")
  (CallFunction|HideRangeVisualization))

(fn UpdateBarrelMeshCollision
  (bind _collisionprofilename (LocalVariables|GetCollisionProfileName))
  (Utilities|FlowControl|Sequence
    (:then_0
      (LocalVariables|SetNoCollisionProfileName (CreativeDevices|GetNamefromCollisionProfile "(Name=\"FortCreativeOnlyVolumeManager\")"))
      (LocalVariables|SetEditModeCollisionProfileName (CreativeDevices|GetNamefromCollisionProfile "(Name=\"FortPhysicsBodyMeshPhysics\")"))
      (LocalVariables|SetPlayModeProfileName (CreativeDevices|GetNamefromCollisionProfile "(Name=\"FortPhysicsBodyMeshPhysics\")"))
      (if (CallFunction|IsinPlayMode)
        (switch Utilities|FlowControl|Switch|SwitchonEnum_ForcedValueDuringGame (Variables|Replicated|GetCollisionForcedDuringGame)
          (:NewEnumerator0
            (switch Utilities|FlowControl|Switch|SwitchonEnum_CollisionDuringGamesSetting (Variables|UserOptions|GetCollision%20During%20Games)
              (:NewEnumerator0
                (LocalVariables|SetCollisionProfileName (LocalVariables|GetNoCollisionProfileName)))
              (:NewEnumerator1
                (LocalVariables|SetCollisionProfileName (LocalVariables|GetPlayModeProfileName)))
              (:NewEnumerator3
                (if (Variables|UserOptions|GetVisible%20During%20Game)
                  (LocalVariables|SetCollisionProfileName (LocalVariables|GetPlayModeProfileName))
                  (else
                    (LocalVariables|SetCollisionProfileName (LocalVariables|GetNoCollisionProfileName)))))))
          (:NewEnumerator1
            (LocalVariables|SetCollisionProfileName (LocalVariables|GetPlayModeProfileName)))
          (:NewEnumerator2
            (LocalVariables|SetCollisionProfileName (LocalVariables|GetNoCollisionProfileName))))
        (else
          (LocalVariables|SetCollisionProfileName (LocalVariables|GetEditModeCollisionProfileName)))))
    (:then_1
      (if (Utilities|Name|NotEqual(Name) _collisionprofilename "None")
        (Collision|SetCollisionProfileName (Variables|Default|GetSM_MilitaryBarrel) _collisionprofilename true :!class /Script/Engine.PrimitiveComponent)
        (CallFunction|UpdateHealthWidget)))))

(fn UpdateVisibilityComponent
  (Visibility|SetVisiblityDuringPhase (Variables|Default|GetVisibleInGameComponent) (select (and (Variables|UserOptions|GetVisible%20During%20Game) (CallFunction|LocalPlayerShouldSeeBarrel)) "NewEnumerator3" "NewEnumerator1")))

(fn ConstructionScript
  (|CallParentFunction)
  (Variables|/Script/Engine.SceneComponent|SetbAbsoluteScale true (Variables|Default|GetRangeSphere))
  (Variables|/Script/Engine.SceneComponent|SetbAbsoluteScale true (Variables|Default|GetProximityTrigger)))
