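; Creative fuel pump device that extends a parent pump with configurable fuel capacity, health, team/class restrictions and explosion damage.
; Option changes and minigame start or end reset fuel, selecting a finite or infinite held-pump class.
; Held-weapon ammo syncs remaining fuel, broadcasting On Empty at zero; Restock, Empty, Enable and Disable receivers control it.
; Cable length and tether range scale with the mesh, and legacy options migrate to current ones.
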(event EventBeginPlay
  ;[0.3012 1 0.107] Update the length of the cable based on the scale of the pump mesh
  (|CallParentFunction)
  (Networking|SwitchHasAuthority
    (:Remote
      (bind _cable (Variables|Default|GetCable)
        (:then
          (Variables|/Script/CableComponent.CableComponent|SetCableLength (* (Variables|/Script/CableComponent.CableComponent|GetCableLength _cable) (CallFunction|GetMeshMaxScale)) _cable))
        (:else)))))

(event OnAnyPropertyChanged(ToyOptionsComponent)
  (bind _creative_enabled_component (Variables|Default|GetCreative_Enabled_Component))
  ;[0.08176 0.7135 0.122] Initialize device
  (CallFunction|InitializeFuel)
  (Class|CreativeEnabledComponent|InitializeComponent _creative_enabled_component (Variables|Default|GetFortMinigameProgress) 2)
  (Variables|/Game/Creative/Devices/Common/Components/Creative_Enabled_Component.Creative_Enabled_Component_C|SetEnabledDuringPlayMode (Variables|UserOptions|Advanced|GetEnabled%20At%20Game%20Start) _creative_enabled_component)
  (Initialization|Initialize (Variables|Default|GetCreative_UseRestriction_Component) (Variables|Default|GetFortMinigameProgress) (Variables|Default|GetCreative_Enabled_Component) (Variables|Default|GetToyOptionsComponent) (Variables|UserOptions|Advanced|GetAllowed%20Team) (Variables|UserOptions|Advanced|GetInvert%20Team%20Selection) (Variables|UserOptions|Advanced|GetSelectedClass) (Variables|UserOptions|Advanced|GetInvert%20Class%20Selection) false)
  (Variables|Default|SetbSpreadFire (Variables|UserOptions|Advanced|GetExplosion%20Deals%20Damage)))

(event OnPropertiesLoaded(ToyOptionsComponent)
  (bind _creative_enabled_component (Variables|Default|GetCreative_Enabled_Component))
  ;[0.08176 0.7135 0.122] Initialize device
  (CallFunction|InitializeFuel)
  (Class|CreativeEnabledComponent|InitializeComponent _creative_enabled_component (Variables|Default|GetFortMinigameProgress) 2)
  (Variables|/Game/Creative/Devices/Common/Components/Creative_Enabled_Component.Creative_Enabled_Component_C|SetEnabledDuringPlayMode (Variables|UserOptions|Advanced|GetEnabled%20At%20Game%20Start) _creative_enabled_component)
  (Initialization|Initialize (Variables|Default|GetCreative_UseRestriction_Component) (Variables|Default|GetFortMinigameProgress) (Variables|Default|GetCreative_Enabled_Component) (Variables|Default|GetToyOptionsComponent) (Variables|UserOptions|Advanced|GetAllowed%20Team) (Variables|UserOptions|Advanced|GetInvert%20Team%20Selection) (Variables|UserOptions|Advanced|GetSelectedClass) (Variables|UserOptions|Advanced|GetInvert%20Class%20Selection) false)
  (Variables|Default|SetbSpreadFire (Variables|UserOptions|Advanced|GetExplosion%20Deals%20Damage)))

(event MinigameStartedDelegate(FortMinigameProgress)
  ;[0.7597 1 0.1786] On Minigame state changed
  (CallFunction|ResetHealthifSet)
  (CallFunction|InitializeFuel))

(event MinigameEndedDelegate(FortMinigameProgress)
  ;[0.7597 1 0.1786] On Minigame state changed
  (CallFunction|ResetHealthifSet)
  (CallFunction|InitializeFuel))

(event EventSetCableAttachedOnServer
  (params (bAttached :type Boolean))
  (bind _playerpawnholdingpumphandle (Variables|Default|GetPlayerPawnHoldingPumpHandle))
  (bind _currentweapon (Variables|/Script/FortniteGame.FortPawn|GetCurrentWeapon _playerpawnholdingpumphandle))
  ;[0.1514 1 0.4128] On pump handle picked up
  (|CallParentFunction bAttached)
  (Utilities|IsValid _playerpawnholdingpumphandle
    (:Is%20Valid
      (Utilities|IsValid _currentweapon
        (:Is%20Valid
          (Variables|Default|SetPumpWeapon _currentweapon)
          (if (Math|Boolean|ANDBoolean bAttached (not (Variables|UserOptions|Basic|GetbInfiniteFuel)) (Networking|HasAuthority))
            (Variables|Default|SetCachedPumpMaxClipSize (FortWeapon|GetBulletsPerClip _currentweapon))
            (EventDispatchers|BindEventtoOnLocalAmmoChanged _currentweapon (EventDispatchers|CreateEvent :!event On%20Fuel%20Pump%20Ammo%20Changed))))) ; This event can onlt get called if \nthe players current weapon is a pump,\nso no cast check is needed here
    )))

(event Interaction|EventBlueprintOnInteract
  (params (InteractingPawn :type Fort%20Pawn%20Object%20Reference) (InteractionBeingAttempted :type EInteractionBeingAttempted%20Enum :default "FirstInteraction"))
  ;[0.2218 1 0.6377] Scale the tether range on the pump handle based on the scale of the pump mesh after the pump is picked up
  (|CallParentFunction InteractingPawn InteractionBeingAttempted)
  (bind _heldpetrolpump (Variables|Default|GetHeldPetrolPump)
    (:then
      (Variables|/Game/Athena/Items/Weapons/Prototype/PetrolPump/BGA_Held_Petrol_Pump.BGA_Held_Petrol_Pump_C|SetTetherRangeScale (CallFunction|GetMeshMaxScale) _heldpetrolpump))
    (:else)))

(event Custom|RestockFunction
  (params (Instigator :type Controller%20Object%20Reference))
  ;[0.4439 0.2916 1] Receivers
  (Variables|Default|SetRemainingFuel (Variables|UserOptions|Basic|GetFuel%20Capacity))
  (bind _pumpweapon (Variables|Default|GetPumpWeapon)
    (:then
      (FortWeapon|ModAmmo _pumpweapon (- (Variables|Default|GetRemainingFuel) (FortWeapon|GetRemainingAmmo _pumpweapon))))
    (:else)) ; If the pump is currently being held, update it's ammo
)

(event Custom|Empty%20Event
  (params (Instigator :type Controller%20Object%20Reference))
  ;[0.4439 0.2916 1] Receivers
  (if (Class|CreativeEnabledComponent|IsEnabled (Variables|Default|GetCreative_Enabled_Component))
    (if (> (Variables|Default|GetRemainingFuel) 0)
      (Variables|Default|SetRemainingFuel 0)
      (GameplayEvent|BroadcastEvent (Variables|UserOptions|Events|GetOn%20Empty) 0 (Utilities|Struct|MakeGameplayEventEmptyPump Instigator))
      (bind _pumpweapon (Variables|Default|GetPumpWeapon)
        (:then
          (FortWeapon|ModAmmo _pumpweapon (- (Variables|Default|GetRemainingFuel) (FortWeapon|GetRemainingAmmo _pumpweapon))))
        (:else)) ; If the pump is currently being held, update it's ammo
    )))

(event Custom|Enable%20Event
  (params (Instigator :type Controller%20Object%20Reference))
  ;[0.4439 0.2916 1] Receivers
  (Class|CreativeEnabledComponent|SetEnabled (Variables|Default|GetCreative_Enabled_Component) true))

(event Custom|Disable%20Event
  (params (Instigator :type Controller%20Object%20Reference))
  ;[0.4439 0.2916 1] Receivers
  (Class|CreativeEnabledComponent|SetEnabled (Variables|Default|GetCreative_Enabled_Component) false)
  (CallFunction|RemovePumpHandleandDetachCable))

(event EventEndPlay
  (params (EndPlayReason :type EEndPlayReason%20Enum :default "Destroyed"))
  ;[1 0.4775 0.05432] Cleanup
  (|CallParentFunction EndPlayReason)
  (CallFunction|RemovePumpHandleandDetachCable))

(fn ApplyDamageToActor
  (params (Actor%20to%20Damage :type Actor%20Object%20Reference))
  ;[1 0.3723 0.05754] Override explosion damage
  (if (Variables|UserOptions|Advanced|GetExplosion%20Deals%20Damage)
    (|CallParentFunction Actor%20to%20Damage)))

(fn BlueprintCanInteract
  (params (InteractingPawn :type Fort%20Pawn%20Object%20Reference) (InteractionBeingAttempted :type EInteractionBeingAttempted%20Enum :default "FirstInteraction") (InteractionType :type TInteractionType%20Enum :default "IT_NoInteraction"))
  (bind _returnvalue (Pawn|GetController InteractingPawn))
  (if (Class|CreativeCommonDeviceFunctionLibrary|IsLogicinPlayMode (Variables|Default|GetFortMinigameProgress))
    ;[0.4247 1 0.3851] In play mode, allow interaction if the player is passes the restriction checks and the parent fuel pump allows it
    (bind _valid_controller (CreativeUserRestrictions|ActorHasValidParticipantController (Variables|Default|GetCreative_UseRestriction_Component) InteractingPawn true false true))
    (return (and _valid_controller (|CallParentFunction InteractingPawn InteractionBeingAttempted InteractionType)))
    (else
      ;[1 0.7462 0.2384] In edit mode, allow interaction if the player has edit permision on the island
      (Utilities|IsValid InteractingPawn
        (:Is%20Valid
          (Utilities|IsValid _returnvalue
            (:Is%20Valid
              (bind _asfort_player_controller_athena (Utilities|Casting|CastToFortPlayerControllerAthena _returnvalue))
              (return (Creative|CanPlayerCreateinCurrentExperience _asfort_player_controller_athena)))))))))

(fn BlueprintGetInteractionString
  (params (InteractingPawn :type Fort%20Pawn%20Object%20Reference) (InteractionBeingAttempted :type EInteractionBeingAttempted%20Enum :default "FirstInteraction"))
  ;[1 0.6587 0.1146] Only show the "Grab Pump" Interaction text during play mode
  (if (Class|CreativeCommonDeviceFunctionLibrary|IsLogicinPlayMode (Variables|Default|GetFortMinigameProgress))
    (bind _returnvalue (|CallParentFunction InteractingPawn InteractionBeingAttempted))
    (return _returnvalue)
    (else
      (return ""))))

(fn GetMeshMaxScale
  (bind (_relativescale3d_x _relativescale3d_y _relativescale3d_z) (Math|Vector|BreakVector (Variables|/Script/Engine.SceneComponent|GetRelativeScale3D_X (Variables|Debug|GetStaticMeshComponent))))
  (return (Math|Float|Max(Float) _relativescale3d_x _relativescale3d_y _relativescale3d_z)))

(fn InitializeFuel
  ;[0.179 1 0.2571] Initialize variables based on the current Fuel Capacity knob value
  (Utilities|FlowControl|Sequence
    (:then_0
      (Variables|Default|SetRemainingFuel (Variables|UserOptions|Basic|GetFuel%20Capacity))
      (Networking|SwitchHasAuthority
        (:Authority
          (Variables|Default|SetHeldPetrolPumpClass (select (Variables|UserOptions|Basic|GetbInfiniteFuel) (Variables|Default|GetInfiniteHeldPumpClass) (Variables|Default|GetFiniteHeldPumpClass))) ; The BR fuel pump has no UI since it has infinite fuel, \nso we made new creative blueprints that allow the pump to \nact similar to how the fuel can in BR does
        )))
    (:then_1
      (CallFunction|RemovePumpHandleandDetachCable))))

(fn IsTargetableByLinearMover
  (return true))

(fn MigrateLegacyUserOptions
  (bind _fuelcapacity (Class|GetClassDefaults "/CreativeCoreDevices/Device_FuelPump_V2.Device_FuelPump_V2_C"))
  (bind _fuelcapacity_2 (Variables|Legacy|GetFuelCapacity))
  (|CallParentFunction)
  (Utilities|FlowControl|Sequence
    (:then_0
      (Utilities|FlowControl|Sequence
        (:then_0
          ;[1 0.3369 0.2672] Migrating Selected class
          (if (!= (Class|GetClassDefaults "/CreativeCoreDevices/Device_FuelPump_V2.Device_FuelPump_V2_C") (Variables|Legacy|GetKnob_SelectedClass))
            (if (== (Variables|Legacy|GetKnob_SelectedClass) -2)
              (Variables|SetSelectedClass_ClassType (struct :ClassType "Any" :ClassSlot 1))
              (Variables|UserOptions|Advanced|SetInvert%20Class%20Selection true)
              (else
                (Variables|UserOptions|Advanced|SetSelectedClass (Creative|ClassOption|MakefromLegacyClassIndex (Variables|Legacy|GetKnob_SelectedClass)))
                (Variables|UserOptions|Advanced|SetInvert%20Class%20Selection false)))
            (Variables|Legacy|SetKnob_SelectedClass (Class|GetClassDefaults "/CreativeCoreDevices/Device_FuelPump_V2.Device_FuelPump_V2_C"))
            (LocalVariables|SetMigrated%20Options true)))
        (:then_1
          ;[0.4931 1 0.2666] Migrating Health / Indestructible
          (if (!= (Class|GetClassDefaults "/CreativeCoreDevices/Device_FuelPump_V2.Device_FuelPump_V2_C") (Variables|Legacy|GetMaxHealth))
            (if (> (Variables|Legacy|GetMaxHealth) 0)
              (Variables|UserOptions|Basic|SetbIndestructible false)
              (else
                (Variables|UserOptions|Basic|SetbIndestructible true)))
            (Variables|UserOptions|Basic|SetHealth (Math|Integer|Clamp(Integer) (Math|Float|Truncate (Variables|Legacy|GetMaxHealth)) 1 50000))
            (Variables|Legacy|SetMaxHealth (Class|GetClassDefaults "/CreativeCoreDevices/Device_FuelPump_V2.Device_FuelPump_V2_C"))
            (LocalVariables|SetMigrated%20Options true)))
        (:then_2
          ;[1 1 1] Migrating Fuel Capacity
          (if (!= _fuelcapacity _fuelcapacity_2)
            (if (> _fuelcapacity_2 0)
              (Variables|UserOptions|Basic|SetbInfiniteFuel false)
              (Variables|UserOptions|Basic|SetFuel%20Capacity (Variables|Legacy|GetFuelCapacity))
              (else
                _))
            (Variables|Legacy|SetFuelCapacity _fuelcapacity) ; Legacy
            (LocalVariables|SetMigrated%20Options true)))))
    (:then_1
      (if (LocalVariables|GetMigrated%20Options)
        (OptionData|RefreshPropertyOverrides (Variables|Default|GetToyOptionsComponent))))))

(fn OnFuelPumpAmmoChanged
  (params (LocalCount :type Integer) (LocalRemaining :type Integer))
  ;[0 1 0.5827] Keep fuel updated while using a finite amount of fuel
  (if (and (== LocalCount (Variables|Default|GetCachedPumpMaxClipSize)) (!= (Variables|Default|GetCachedPumpMaxClipSize) (Variables|Default|GetRemainingFuel)))
    ;;[0.1619 1 0.1942] If the held pump's ammo is at the max clip size, we need to update it to what the remaining fuel is
    (bind _pumpweapon (Variables|Default|GetPumpWeapon)
      (:then
        (FortWeapon|ModAmmo _pumpweapon (- (Variables|Default|GetRemainingFuel) (Variables|GetLocalRemaining))))
      (:else))
    (else
      ;;[0 0.8825 1] Update the remaining fuel after the held pump uses fuel
      (bind _output_get (Variables|Default|SetRemainingFuel (Variables|GetLocalRemaining)))
      (if (== _output_get 0)
        (Utilities|FlowControl|Sequence
          (:then_0
            (bind _playerpawnholdingpumphandle (Variables|Default|GetPlayerPawnHoldingPumpHandle)
              (:then
                (LocalVariables|SetPlayerControllerWithPump (Pawn|GetController _playerpawnholdingpumphandle)))
              (:else)))
          (:then_1
            (GameplayEvent|BroadcastEvent (Variables|UserOptions|Events|GetOn%20Empty) 0 (Utilities|Struct|MakeGameplayEventEmptyPump (LocalVariables|GetPlayerControllerWithPump))))) ; The instigator doesn't have to be valid,\nwe always want to trigger if the fuel is empty 
      ))))

(fn ResetHealthIfSet
  (bind _self self)
  (bind _returnvalue (Ability|GetAbilitySystemComponent _self))
  (bind _health (Variables|UserOptions|Basic|GetHealth))
  (bind _returnvalue_2 (> _health 0))
  ;[1 0.5167 0.1567] If the Max Health is negative, make the device indestructible
  (if (not (Variables|UserOptions|Basic|GetbIndestructible))
    (CallFunction|SetSuppressHealthBar (not _returnvalue_2))
    (Fortnite|SetCanbeDamaged self _returnvalue_2)
    (LocalVariables|SetHealth_OnFunction (select _returnvalue_2 (Math|Conversions|ToFloat(Integer) (Variables|UserOptions|Basic|GetHealth)) 1e+03))
    ;[1 0.8159 0.2634] Applying GE to modify maximum health on this object.
    (Networking|SwitchHasAuthority
      (:Authority
        (Utilities|IsValid _returnvalue
          (:Is%20Valid
            (bind _returnvalue_3 (Ability|GameplayEffect|AssignTagSetbyCallerMagnitude (GameplayEffects|MakeOutgoingSpec :self _returnvalue :GameplayEffectClass "/Game/Creative/Devices/Common/GameplayEffects/GE_Devices_Common_SetMaxHealth.GE_Devices_Common_SetMaxHealth_C" :Level 0.0) "(TagName=\"SetByCaller.MaxHealth\")" (LocalVariables|GetHealth_OnFunction)))
            (GameplayEffects|RemoveActiveGameplayEffect _returnvalue (Variables|Default|GetMaxHealthEffectHandler) -1)
            (bind _returnvalue_4 (GameplayEffects|ApplyGameplayEffectSpecToSelf _returnvalue _returnvalue_3))
            (Variables|Default|SetMaxHealthEffectHandler _returnvalue_4)
            (Health|SetHealth (Health|GetMaxHealth))))))))

(fn ShouldInvisiblePropsBeAllowed
  (return false))

(fn ConstructionScript
  (|CallParentFunction))
