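; Creative HUD Message device that shows configurable on-screen messages to players filtered by team or instigator, on round start or trigger.
; The server builds message payloads and sends them to per-layer components on each PlayerController, created on demand.
; It tracks active and queued messages for late joiners and team switches, and syncs same-layer clears through a gameplay event router.
; It also supports Verse text and duration overrides and migrates legacy options.
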(event EventOnMinigameStarted
  ;[0 1 0.5103 0.2] Initialization Logic
  (|CallParentFunction)
  (Networking|SwitchHasAuthority
    (:Authority
      (CallFunction|GenerateGuid)
      (CallFunction|ShowOnMinigameStartIfSet)
      (if (Variables|Transient|GetbShowWhenMinigameAvailable)
        (CallFunction|ShowMessage (Variables|Default|GetLastInstigator))))))

(event EventInitializeDevice
  ;[0 1 0.5103 0.2] Initialization Logic
  (|CallParentFunction)
  (UserOptions|MigrateAfterInitialize) ; Text needs migrated after options are loaded and not before.
  (UserOptions|MigrateRichTextStyle) ; Fix FORT-870488 to use legacy rich text style set if using the legacy syntax
)

(event OnAddedtoMinigame(FortMinigameProgress)
  (params (Minigame :type Fort%20Minigame%20Object%20Reference))
  (bind _router (CallFunction|GetEventRouter))
  ;[0 1 0.5103 0.2] Initialization Logic
  (Networking|SwitchHasAuthority
    (:Authority
      (Minigame|BindEventtoPlayerChangedTeamDelegate Minigame (delegate Custom|OnPlayersTeamSwitch))
      (Minigame|BindEventtoPlayerAddedDelegate Minigame (delegate Custom|On%20Player%20Added))
      ;;[1 1 1] Any HUD Message can send a message to clear all screen messages. When this happens, all devices clear their own messages.
      (Utilities|IsValid _router
        (:Is%20Valid
          (bind (_asynctaskproxy _payload) (GameplayEventRouter|StartListeningToEvent _router "/CreativeCoreDevices/HUDMessage/GameplayEvent_HUDMessageClearAll.GameplayEvent_HUDMessageClearAll")
            (:OnEventReceived
              (if (== (Variables|UserOptions|GetLayer) (.Layer_12_EF321C4341237E02D67EE18B94187A20 (Utilities|Struct|BreakGameplayEventHUDMessageClearAll _payload)))
                (Utilities|IsValid (Utilities|Struct|BreakGameplayEventHUDMessageClearAll _payload)
                  (:Is%20Valid
                    (bind _true (CallFunction|IsRelevantForPlayer (Utilities|Struct|BreakGameplayEventHUDMessageClearAll _payload)))
                    (if _true
                      (HUDMessageUpdates|ClearMessage))))))))))))

(event Custom|On%20Player%20Added
  (params (UniqueNetId :type Unique%20Net%20Id%20Structure) (bIsLocalPlayer :type Boolean))
  (bind _returnvalue (Game|GetPlayerStatefromUniqueNetId UniqueNetId))
  (bind _returnvalue_2 (Fortnite|GetFortPlayerControllerfromActor _returnvalue))
  ;[0 1 0.5103 0.2] Initialization Logic
  ;;[0 1 0.3421 0] Join in Progress - Test to see if this message is active and should be in the player's queue.
  (Utilities|MessageIsActiveforJIP
    (:Yes
      (Utilities|IsValid _returnvalue
        (:Is%20Valid
          (bind _true (CallFunction|IsRelevantForPlayer _returnvalue_2))
          (if _true
            (bind _comp (CallFunction|GetPCComp _returnvalue_2))
            (bind (_payload _last_instigator_player_state) (CallFunction|GetWidgetDetails))
            (Server|HUDMessageRequest _comp self true (Utilities|Struct|MakeHUDMessageQueueDetails (Utilities|Time|GetGameTimeinSeconds) _last_instigator_player_state _payload))))))))

(event Custom|OnPlayersTeamSwitch
  (params (PlayerState :type Fort%20Player%20State%20Object%20Reference) (TeamId :type Byte) (OldTeamId :type Byte) (bTeamChangedWithoutRespawn :type Boolean))
  ;[0 1 0.5103 0.2] Initialization Logic
  ;;[0 0.4072 0.1384 0] Bind Minigame to Player Changed Team - Show all messages that were display for the team switched
  (if (Variables|UserOptions|GetRe-Evaluate%20Messages%20on%20Show)
    (CallFunction|ShowQueuedMessagesonTeamSwitch (Fortnite|GetFortPlayerControllerfromActor PlayerState))))

(event EventOnMinigameEnded
  ;[1 0.06879 0 0.2] End Minigame
  (|CallParentFunction)
  (Utilities|Time|ClearandInvalidateTimerbyHandle (Variables|Transient|GetShowMessageTimerHandle)) ; Make sure no messages play after the round / game is over
  (Utilities|Map|Clear (Variables|Default|GetPlayerControllerComponents))
  (Default|ResetOverrideVariables)
  (HUDMessageUpdates|ClearMessage))

(event Custom|Show%20on%20Minigame%20Start%20if%20Set
  (bind _timefromroundstart (Variables|UserOptions|Basic|GetTimeFromRoundStart))
  ;[0 1 0.9627 0.2] Timer Logic
  (if (Variables|UserOptions|Basic|GetShowOnRoundStart)
    (if (<= _timefromroundstart 0)
      (CallFunction|ShowMessage 0)
      (else
        (bind _returnvalue (Utilities|Time|SetTimerbyEvent (delegate Custom|ShowMessageFromTimer) _timefromroundstart))
        (Variables|Transient|SetShowMessageTimerHandle _returnvalue)))))

(event Custom|ShowWhenReceivingFrom
  (params (Instigator :type Controller%20Object%20Reference))
  ;[0.1947 0 1 0.25] Gameplay Functions
  (CallFunction|ShowMessage Instigator)
  (GameplayEvent|BroadcastEvent (Variables|UserOptions|Events|GetShowEvent) 0 (Utilities|Struct|MakeGameplayEventHUDMessageShow Instigator)))

(event Custom|ShowMessageFromTimer
  ;[0 1 0.9627 0.2] Timer Logic
  (CallFunction|ShowMessage 0))

(event Custom|HideWhenReceivingFrom
  (params (Instigator :type Controller%20Object%20Reference))
  ;[0.1947 0 1 0.25] Gameplay Functions
  (HUDMessageUpdates|UpdateforActors Instigator false)
  (GameplayEvent|BroadcastEvent (Variables|UserOptions|Events|GetHideEvent) 0 (Utilities|Struct|MakeGameplayEventHUDMessageHide Instigator)))

(event Custom|ShowMessage
  (params (TriggerInstigator :type Controller%20Object%20Reference))
  ;[0.1332 0.5271 1 0.2] Show Message Logic
  (Variables|Default|SetLastInstigator TriggerInstigator)
  (Utilities|IsinProgress?
    (:Yes
      (HUDMessageUpdates|UpdateforActors TriggerInstigator)
      (Variables|Transient|SetbShowWhenMinigameAvailable false))
    (:No%20Minigame
      (Variables|Transient|SetbShowWhenMinigameAvailable true))))

(event Custom|ClearAllWhenReceivingFrom
  (params (Instigator :type Controller%20Object%20Reference))
  ;[0.1947 0 1 0.25] Gameplay Functions
  (HUDMessageUpdates|ClearAllMessages Instigator)
  (GameplayEvent|BroadcastEvent (Variables|UserOptions|Events|GetClearedEvent) 0 (Utilities|Struct|MakeGameplayEventHUDMessageClearMessages Instigator)))

(event Custom|UpdateMessageQueue
  (bind _messagequeue (Variables|Default|GetMessageQueue))
  (bind _output (Utilities|Array|Get(aref) _messagequeue))
  ;[0.1485 0.7887 1 0.482] Update local message queue for players that switch teams
  (if (Utilities|Array|IsNotEmpty (Variables|Default|GetMessageQueue))
    (Utilities|FlowControl|Delay (Utilities|Struct|BreakGameplayEventHUDMessageDetails (Utilities|Struct|BreakHUDMessageQueueDetails _output)))
    (Utilities|Array|RemoveItem _messagequeue _output)
    (CallFunction|UpdateMessageQueue)
    (else
      (Variables|Default|SetIsQueueUpdating false))))

(event HandleVersionUpgrades(CreativeVersion)
  (params (LastSavedVersion :type Creative%20Version%20Id%20Structure))
  ;[1 1 1] FORT-902219: Adding a version difference to stop ShowMessage from happening on newer versions when calling SetText in Verse
  (CreativeVersion|CompareVersiontoTag LastSavedVersion "(TagName=\"Creative.Version.Release.37.00\")"
    (:OlderVersion
      (Variables|Default|SetShowTextOnOverride true))))

(fn ClearAllMessages
  (params (Triggering%20Player :type Controller%20Object%20Reference))
  (bind _router (CallFunction|GetEventRouter))
  (bind _playerstates (CallFunction|GetRelevantPlayers Triggering%20Player))
  (for _array_element _playerstates
    (bind _comp (CallFunction|GetPCComp (PlayerState|GetPlayerController _array_element)))
    (Class|CRDHUDMessageComp|ClearAllMessages _comp)
    (Utilities|Array|Clear (Variables|Default|GetMessageQueue))
    (Variables|Default|SetIsQueueUpdating false))
  (Utilities|IsValid _router
    (:Is%20Valid
      (GameplayEventRouter|BroadcastEvent _router (Utilities|Struct|MakeGameplayEventHUDMessageClearAll Triggering%20Player (Variables|UserOptions|GetLayer)) 0))))

(fn ClearMessage
  (Variables|Default|SetLastInstigator 0)
  (Variables|Default|SetActiveTime 0.0))

(fn DisplayMessage
  (params (Controller :type Controller%20Object%20Reference) (Message :type Localizable%20Message%20Structure) (DisplayTime :type Float%20%28double-precision%29))
  (bind _message_72_7634162b4c068206f407bcac0c836150 (Utilities|Struct|BreakGameplayEventHUDMessageDetails (LocalVariables|GetLocal%20Payload)))
  (bind (_message_72_7634162b4c068206f407bcac0c836150_text_2_563e0da1490d74cee84ef39a5df6bb26 _message_72_7634162b4c068206f407bcac0c836150_message_5_ddd38afe499b5ae0e2151a99b81ee564 _message_72_7634162b4c068206f407bcac0c836150_charlimit_11_f93cc8954d727984a85aacb9929a1685) (Utilities|Struct|BreakLocalizablePayload _message_72_7634162b4c068206f407bcac0c836150))
  (Networking|SwitchHasAuthority
    (:Authority
      (Utilities|IsValid Controller
        (:Is%20Valid
          ;[1 1 1] The Controller passed in comes from Verse and is provided as an Agent. It's not necessarily a valid PlayerController. So we need to get the real PlayerController by going through the PlayerState.
          (bind _asfort_player_state (Utilities|Casting|CastToFortPlayerState (Variables|/Script/Engine.Controller|GetPlayerState Controller)))
          (LocalVariables|SetLocal%20Controller (PlayerState|GetPlayerController _asfort_player_state)))
        (:Is%20Not%20Valid
          ;[1 0.6634 0 0.2] If no Controller specified, this message is for everyone. Iterate the message through all players.
          (bind _fortminigameprogress (Variables|Default|GetFortMinigameProgress)
            (:then
              (bind _currentminigame (Variables|/Script/FortniteGame.FortMinigameLogicComponent|GetCurrentMinigame _fortminigameprogress)
                (:then
                  (for _array_element (Minigame|GetParticipatingPlayers _currentminigame)
                    (Utilities|IsValid _array_element
                      (:Is%20Valid
                        (LocalVariables|SetLocal%20Controller (PlayerState|GetPlayerController _array_element))
                        (bind _comp (CallFunction|GetPCComp (LocalVariables|GetLocal%20Controller)))
                        (bind _payload (CallFunction|GetWidgetDetails))
                        (bind _output_get (LocalVariables|SetLocal%20Payload _payload))
                        (bind _structout (Utilities|Struct|SetmembersinGameplayEventHUDMessageDetails _output_get (Utilities|Struct|MakeLocalizablePayload _message_72_7634162b4c068206f407bcac0c836150_text_2_563e0da1490d74cee84ef39a5df6bb26 (Variables|UserOptions|Basic|GetMessage) _message_72_7634162b4c068206f407bcac0c836150_charlimit_11_f93cc8954d727984a85aacb9929a1685)))
                        (LocalVariables|SetLocal%20Payload _structout)
                        (if (>= (Variables|UserOptions|GetDisplayTime) 0)
                          (bind _structout_2 (Utilities|Struct|SetmembersinGameplayEventHUDMessageDetails (LocalVariables|GetLocal%20Payload) (Math|Float|SelectFloat -1.0 (Math|Float|Clamp(Float) (Variables|UserOptions|GetDisplayTime) 0.0 6e+01) (== (Variables|UserOptions|GetDisplayTime) 0))))
                          (LocalVariables|SetLocal%20Payload _structout_2)
                          (else
                            _))
                        (Server|HUDMessageRequest _comp self true (Utilities|Struct|MakeHUDMessageQueueDetails (Utilities|Time|GetGameTimeinSeconds) (Helpers|GetPlayerStatefromController (Variables|GetController)) (LocalVariables|GetLocal%20Payload)))))))
                (:else)))
            (:else)))))))

(fn GenerateGuid
  ;[0 0.3279 1 0.2] Generate a GUID for this object. This will be needed to check against messages on the client.
  (bind _returnvalue (GUID|GenerateGuidfromObject self true))
  (Variables|Default|SetMessageGUID _returnvalue))

(fn GetEventRouter
  (bind _routercached (Variables|Default|GetRouterCached)
    (:then
      (return _routercached))
    (:else
      (bind _returnvalue (GameplayEventRouter|GetEventRouter "/Script/FortniteGame.FortPlayspace" self))
      (bind _output_get (Variables|Default|SetRouterCached _returnvalue))
      (return _output_get))))

(fn GetPCComp
  (params (Actor :type Actor%20Object%20Reference) (CreateIfNeeded :type Boolean :default true))
  (bind _pc (LocalVariables|GetPC))
  (bind _playercontrollercomponents (Variables|Default|GetPlayerControllerComponents))
  (bind _value (Utilities|Map|Find _playercontrollercomponents _pc))
  (Networking|SwitchHasAuthority
    (:Authority
      (bind _output_get (LocalVariables|SetPC (Fortnite|GetFortPlayerControllerfromActor Actor)))
      (Utilities|IsValid _output_get
        (:Is%20Valid
          (Utilities|IsValid _value
            (:Is%20Valid
              (return _value))
            (:Is%20Not%20Valid
              (bind _output_get_2 (LocalVariables|Setcached%20comps (Actor|GetComponentsByClass (LocalVariables|GetPC) "/CreativeCoreDevices/HUDMessage/CRD_HUDMessageComp.CRD_HUDMessageComp_C")))
              (for _array_element _output_get_2
                (Utilities|IsValid _array_element
                  (:Is%20Valid
                    (if (== (Variables|/CreativeCoreDevices/HUDMessage/CRD_HUDMessageComp.CRD_HUDMessageComp_C|GetHUD%20Layer _array_element) (Variables|UserOptions|GetLayer))
                      (Utilities|Map|Add (Variables|Default|GetPlayerControllerComponents) (LocalVariables|GetPC) _array_element)
                      (return _array_element)))))
              (if (Variables|GetCreateIfNeeded)
                (bind _returnvalue (AddComponent|NODE_AddCRD_HUDMessageComp-0 :self (LocalVariables|GetPC) :HUD%20Layer (Variables|UserOptions|GetLayer)))
                (Utilities|Map|Add (Variables|Default|GetPlayerControllerComponents) (LocalVariables|GetPC) _returnvalue)
                (return _returnvalue)
                (else
                  (return 0))))))
        (:Is%20Not%20Valid
          _)))
    (:Remote
      _))
  (return 0))

(fn GetPresetPlacement
  (params (Key :type EHUDMessagePlacementV2%20Enum :default "NewEnumerator0") (CustomOffset :type Vector%202D%20Structure) (CustomAnchor :type EUIAnchor%20Enum :default "NewEnumerator4"))
  (bind _returnvalue (TouchPlatform|IsMobileGame))
  (bind _mobilepresetlocations (Variables|Default|GetMobilePresetLocations))
  (bind _desktoppresetlocations (Variables|Default|GetDesktopPresetLocations))
  (bind _returnvalue_2 (select _returnvalue _mobilepresetlocations _desktoppresetlocations))
  (bind (_value _returnvalue_3) (Utilities|Map|Find _returnvalue_2 Key))
  (bind (_value_offset_12_431d154b4f705260a9286ea6ecf64f6d _value_anchor_10_091a4a124d9e17e3ebc635bfe8d83ac2) (Utilities|Struct|BreakHUDMessageAlignment _value))
  (if _returnvalue_3
    (return _value_anchor_10_091a4a124d9e17e3ebc635bfe8d83ac2 _value_offset_12_431d154b4f705260a9286ea6ecf64f6d)
    (else
      (return (Variables|GetCustomAnchor) (Variables|GetCustomOffset)))))

(fn GetRelevantPlayers
  (params (Triggering%20Player :type Controller%20Object%20Reference))
  (bind _message_recipient_teamtype (Utilities|Struct|BreakCreativeTeamOption (Variables|GetMessage%20Recipient_TeamType)))
  (bind _currentminigame (Variables|/Script/FortniteGame.FortMinigameLogicComponent|GetCurrentMinigame (Variables|Default|GetFortMinigameProgress))
    (:then
      (bind _output_get (LocalVariables|SetMinigame _currentminigame))
      (Utilities|IsValid (Minigame|GetMinigameTeamComponent _output_get)
        (:Is%20Valid
          (LocalVariables|SetLocalMinigameTeamComponent (Minigame|GetMinigameTeamComponent _output_get))
          ;[0 0.774 1 0.2] Different team types here have significantly different requirements, so switching to run logic based on recipient choice.
          (switch Utilities|FlowControl|Switch|SwitchonECreativeTeamType _message_recipient_teamtype
            (:TeamIndex
              ;[0 0.774 1 0.2] All Members of a specific team.
              (return (Class|FortMinigameTeamComponent|GetTeamMembers (LocalVariables|GetLocalMinigameTeamComponent) (Creative|TeamOption|ConverttoTeam (Variables|UserOptions|GetMessage%20Recipient)))))
            (:All
              ;[0 0.774 1 0.2] All Players
              (return (Minigame|GetParticipatingPlayers (LocalVariables|GetMinigame))))
            (:Friendlies
              ;[0 0.774 1 0.2] All Members on the player's team.
              (bind _triggering_player (Variables|GetTriggering%20Player)
                (:then
                  (return (Class|FortMinigameTeamComponent|GetTeamMembers (LocalVariables|GetLocalMinigameTeamComponent) (Team|GetActorTeam _triggering_player))))
                (:else)))
            (:Enemies
              ;[0 0.774 1 0.2] All players not on the player's team. We're doing this by finding the player's teammates and checking the difference between that set and a set of all players.
              (bind _triggering_player_2 (Variables|GetTriggering%20Player)
                (:then
                  (Utilities|Set|AddItems (LocalVariables|GetSetB) (Class|FortMinigameTeamComponent|GetTeamMembers (LocalVariables|GetLocalMinigameTeamComponent) (Team|GetActorTeam _triggering_player_2)))
                  (Utilities|Set|AddItems (LocalVariables|GetSetA) (Minigame|GetParticipatingPlayers (LocalVariables|GetMinigame)))
                  (bind _result (Utilities|Set|Difference (LocalVariables|GetSetA) (LocalVariables|GetSetB)))
                  (bind _result_2 (Utilities|Set|ToArray _result))
                  (return _result_2))
                (:else)))
            (:TriggeringPlayer
              ;[0 0.774 1 0.2] We just want the triggering player, so early out.
              (bind _triggering_player_3 (Variables|GetTriggering%20Player)
                (:then
                  (bind _asfort_player_state (Utilities|Casting|CastToFortPlayerState (Variables|/Script/Engine.Controller|GetPlayerState _triggering_player_3)))
                  (return (Utilities|Array|MakeArray _asfort_player_state)))
                (:else
                  (return 0))))))))
    (:else)))

(fn GetDefaultStyleOverride
  ;[1 0.6099 0 0.2] Returns updates to the default style. Used by the User Option widgets to show previews of the text style.
  (return (Variables|UserOptions|GetOverride%20Default%20Text%20Style) (Variables|UserOptions|GetTextColor) (Variables|UserOptions|GetShadowOffset) (Variables|UserOptions|GetSize) (Variables|UserOptions|GetOutlineStrength)))

(fn GetDisplayTime
  (return (or (Variables|UserOptions|GetShowForDuration) (> (Variables|Default|GetDisplayTimeOverride) 0)) (Math|Float|Min(Float) (Math|Float|SelectFloat (Variables|Default|GetDisplayTimeOverride) (Math|Float|SelectFloat (Variables|UserOptions|GetDisplayTime) -1.0 (Variables|UserOptions|GetShowForDuration)) (!= (Variables|Default|GetDisplayTimeOverride) 0)) (Variables|Default|GetMax%20Display%20Time))))

(fn GetFiniteDisplayDuration
  (bind (_has_duration _returnvalue) (Default|GetDisplayTime))
  (return _returnvalue))

(fn GetMessage
  (return (Utilities|String|ToString(Text) (select (LocalizableMessage|IsEmptyLocalizableMessage (Variables|Verse|GetMessageOverride)) (Variables|UserOptions|Basic|GetMessage) (LocalizableMessage|ConvLocalizableMessagetoText (Variables|Verse|GetMessageOverride))))))

(fn GetMessagePayload
  (return (Utilities|Struct|MakeLocalizablePayload (Variables|UserOptions|Basic|GetMessage) (Variables|Verse|GetMessageOverride) (Variables|Constants|GetMessageCharLimit))))

(fn GetPlayerStateFromController
  (params (Controller :type Controller%20Object%20Reference))
  (Utilities|IsValid Controller
    (:Is%20Valid
      (return (Variables|/Script/Engine.Controller|GetPlayerState Controller)))
    (:Is%20Not%20Valid
      (return 0))))

(fn GetTextStyle
  (bind _outrownames (DataTable|GetDataTableRowNames "/Game/Athena/HUD/TextStyles/AthenaScoreAlertStyles.AthenaScoreAlertStyles"))
  (for _array_element _outrownames
    (if (Utilities|String|Equal,CaseInsensitive(String) (Utilities|Select :Index (Variables|Legacy|GetText%20Style) :NewEnumerator0 "Score" :NewEnumerator7 "RespawnAlert" :NewEnumerator8 "ScoreCoinsWhite" :NewEnumerator9 "Heist" :NewEnumerator10 "IntroScore" :NewEnumerator11 "Ally" :NewEnumerator12 "Intro" :NewEnumerator13 "Enemy" :NewEnumerator14 "HeistEnemy" :NewEnumerator15 "RespawnAlertB" :NewEnumerator16 "RespawnAlertBLrg") (Utilities|String|ToString(Name) _array_element))
      (return (Utilities|String|ToString(Name) _array_element))))
  (return ""))

(fn GetTextStyleOverride
  ;[1 0.6099 0 0.2] Get the Text Style Override. Used by the UserOption to show the override.
  ; Forcing to always use an override and use the default if not overriding until Popup Dialog can use the new defaults.
  (return true (select (Variables|LegacyMigration|GetIsUsingLegacyRichTextStyle) "/CreativeCoreDevices/Athena/UI/InGame/Creative/PropertyEditor/StyleEditor/RichTextStyleSets/CreativeRichTextStyles_LegacyHUD.CreativeRichTextStyles_LegacyHUD" (select (Variables|UserOptions|GetTextStyleSet_Override) (Variables|UserOptions|GetTextStyleSet) "/CreativeCoreDevices/Athena/UI/InGame/Creative/PropertyEditor/StyleEditor/RichTextStyleSets/CreativeRichTextStyles_Default.CreativeRichTextStyles_Default"))))

(fn GetTimeIntervalUntilMessageIsShown
  (return (- (Variables|Replication|GetServerTimeWhenMessgeWillShow) (Fortnite|Sync|Time|GetServerWorldTimeSeconds))))

(fn GetWidgetDetails
  (bind (_backgroundcolor_r _backgroundcolor_g _backgroundcolor_b _backgroundcolor_a) (Math|Color|BreakColor (Variables|GetBackgroundColor_R)))
  (bind (_has_duration _returnvalue) (Default|GetDisplayTime))
  (bind (_hasoverride _richtextstyle) (TextStyleOverride|GetTextStyleOverride))
  (bind (_anchor _offset) (CallFunction|GetPresetPlacement (Variables|UserOptions|GetPlacement) (Math|Vector2D|MakeVector2D (Variables|UserOptions|GetPlacementHorizontal) (Variables|UserOptions|GetPlacementVertical)) (Variables|UserOptions|GetScreenAnchor)))
  (return (Variables|UserOptions|GetLayer) (Variables|UserOptions|GetHUD%20Widget) (CallFunction|GetMessagePayload) _anchor _offset (Variables|UserOptions|GetText%20Justification) (Creative|UI|IconPicker|GetIconLibraryTexture (Variables|UserOptionsWIP|GetIcon) true) (Variables|UserOptions|GetPlaySound) _returnvalue (Variables|UserOptions|GetIntro%20Animation) (Variables|UserOptions|GetOutro%20Animation) _richtextstyle (Utilities|Struct|MakeHUDMessageRichTextOverrides (Variables|UserOptions|GetOverride%20Default%20Text%20Style) (Variables|UserOptions|GetTextColor) (Variables|UserOptions|GetSize) (Variables|UserOptions|GetOutlineStrength) (Math|Conversions|ToVector2D (Variables|UserOptions|GetShadowOffset))) (Math|Color|MakeColor _backgroundcolor_r _backgroundcolor_g _backgroundcolor_b (Utilities|Operators|Multiply _backgroundcolor_a (Variables|UserOptions|Basic|GetBackgroundOpacity) 0.01)) (Utilities|String|SelectString (Variables|UserOptions|GetVerseTextStyle) "" (Variables|UserOptions|GetVerseTextStyle_Override)) (Helpers|GetPlayerStatefromController (Variables|Default|GetLastInstigator))))

(fn HideForActors
  (params (Relevant%20PC :type Controller%20Object%20Reference))
  (CallFunction|GetPCComp Relevant%20PC))

(fn IsRelevantforPlayer
  (params (Test%20Player :type Controller%20Object%20Reference) (Test%20Instigator :type Boolean) (Instigator%20%28Optional%29 :type Controller%20Object%20Reference))
  (bind _test_player (Variables|GetTest%20Player))
  (bind _teamindex (Team|GetActorPvPTeamIndex _test_player))
  (bind _message_recipient_teamtype (Utilities|Struct|BreakCreativeTeamOption (Variables|GetMessage%20Recipient_TeamType)))
  (Utilities|IsValid Test%20Player
    (:Is%20Valid
      (Utilities|FlowControl|Sequence
        (:then_0
          (bind _currentminigame (Variables|/Script/FortniteGame.FortMinigameLogicComponent|GetCurrentMinigame (Variables|Default|GetFortMinigameProgress))
            (:then
              (LocalVariables|SetMinigame _currentminigame)
              (LocalVariables|SetInstigatorToTest (select (Variables|GetTest%20Instigator) (Variables|GetInstigator%20%28Optional%29) (Variables|Default|GetLastInstigator)))
              (switch Utilities|FlowControl|Switch|SwitchonECreativeTeamType _message_recipient_teamtype
                (:TeamIndex
                  (CreativeTeam|ActorMatchesTeam (Variables|GetTest%20Player) (Variables|UserOptions|GetMessage%20Recipient)
                    (:Yes
                      (return true))))
                (:All
                  (return true))
                (:Friendlies
                  (bind _instigatortotest (LocalVariables|GetInstigatorToTest)
                    (:then
                      (return (== (Team|GetActorPvPTeamIndex _instigatortotest) _teamindex)))
                    (:else)))
                (:Enemies
                  (bind _instigatortotest_2 (LocalVariables|GetInstigatorToTest)
                    (:then
                      (return (not (== _teamindex (Team|GetActorPvPTeamIndex _instigatortotest_2)))))
                    (:else)))))
            (:else)))
        (:then_1
          (return false))))))

(fn IsDisplayDurationFinite
  (return (or (Variables|UserOptions|GetShowForDuration) (> (Variables|Default|GetDisplayTimeOverride) 0))))

(fn MigrateAfterInitialize
  ;[0.8229 0.01059 1 0] We need to do manual generic migration after we know that options are loaded, since the built in MigrateLegacyUserOptions function will fire before that point.
  (Networking|SwitchHasAuthority
    (:Authority
      (if (not (Variables|LegacyMigration|GetMigratedTextInfo))
        (Utilities|FlowControl|Sequence
          (:then_0
            (Utilities|FlowControl|Sequence
              (:then_0
                ;;[1 0 0.5026 0] If text isn't empty, take the pre-existing text style and embed it into the text.
                (if (not (Utilities|Text|TextIsEmpty (Variables|UserOptions|Basic|GetMessage)))
                  (bind _textstyle (UserOptions|GetTextStyle))
                  (LocalVariables|SetTextStyle _textstyle)
                  (Utilities|Text|EditTextSourceString 0 (Variables|UserOptions|Basic|GetMessage) (Utilities|String|ToString(Text) (Utilities|Text|FormatText "NSLOCTEXT(\"[0C14953E4991900ABBD001A3FAA1181D]\", \"312B8E4F4896FDF1354046AA0E4BCE3F\", \"<{Style}>{Message}</>\")" (LocalVariables|GetTextStyle) (Variables|UserOptions|Basic|GetMessage))) false)
                  (Variables|UserOptions|SetVerseTextStyle_Override true)
                  (Variables|UserOptions|SetVerseTextStyle (LocalVariables|GetTextStyle))
                  (Variables|Legacy|SetText%20Style (Class|GetClassDefaults "/CreativeCoreDevices/Device_HUDMessage_V2.Device_HUDMessage_V2_C"))
                  (Variables|UserOptions|SetTextStyleSet "/CreativeCoreDevices/Athena/UI/InGame/Creative/PropertyEditor/StyleEditor/RichTextStyleSets/CreativeRichTextStyles_LegacyHUD.CreativeRichTextStyles_LegacyHUD") ; Use the legacy style set if we're the old version.
                  (Variables|UserOptions|SetTextStyleSet_Override true)
                  (LocalVariables|SetOptionMigrated true)))
              (:then_1
                ;;[1 0.1013 0.7612 0.11] If we're using a custom widget, disable animations (since they now work but didn't before)
                (if (not (Utilities|Set|ContainsItem (LocalVariables|GetCreativeWidgets) (Variables|UserOptions|GetHUD%20Widget)))
                  (Variables|UserOptions|SetIntro%20Animation "NewEnumerator0")
                  (Variables|UserOptions|SetOutro%20Animation "NewEnumerator0")
                  (LocalVariables|SetOptionMigrated true)))))
          (:then_2
            (Variables|LegacyMigration|SetMigratedTextInfo true)
            (if (LocalVariables|GetOptionMigrated)
              (OptionData|RefreshPropertyOverrides (Variables|Default|GetToyOptionsComponent)))))))))

(fn MigrateLegacyUserOptions
  (bind _soundoption (Class|GetClassDefaults "/CreativeCoreDevices/Device_HUDMessage_V2.Device_HUDMessage_V2_C"))
  (bind _displaytimeoption (Class|GetClassDefaults "/CreativeCoreDevices/Device_HUDMessage_V2.Device_HUDMessage_V2_C"))
  (bind _styleoption (Class|GetClassDefaults "/CreativeCoreDevices/Device_HUDMessage_V2.Device_HUDMessage_V2_C"))
  (bind _message_priority (Variables|LegacyMigration|GetMessage%20Priority))
  (bind _timefromroundstartoption (Variables|LegacyMigration|GetTimeFromRoundStartOption))
  (bind _displaytimeoption_2 (Variables|LegacyMigration|GetDisplayTimeOption))
  (bind _styleoption_2 (Variables|LegacyMigration|GetStyleOption))
  (|CallParentFunction)
  (Networking|SwitchHasAuthority
    (:Authority
      (Utilities|FlowControl|Sequence
        (:then_0
          (Utilities|FlowControl|Sequence
            (:then_0
              ;[1 0.1681 0.9838 0.25] Migrating Display Time - Splitting into 'Show for Duration' Bool and Float
              (if (!= _displaytimeoption_2 _displaytimeoption)
                (bind _output_get (Variables|UserOptions|SetShowForDuration (> _displaytimeoption_2 0))) ; If Display Time Option was < 0, hide the\ntime duration by setting bool to false.
                (if _output_get
                  (Variables|UserOptions|SetDisplayTime _displaytimeoption_2) ; Set our display time from the \nold variable.
                  (else
                    _))
                (Variables|LegacyMigration|SetDisplayTimeOption _displaytimeoption) ; Set old variable to default
                (LocalVariables|SetOption%20Migrated true)))
            (:then_1
              ;[1 0.1681 0.9838 0.25] Migrating Time From Round Start - Splitting into 'Show on Round Start' Bool and Float.
              (if (!= (Variables|LegacyMigration|GetTimeFromRoundStartOption) (Class|GetClassDefaults "/CreativeCoreDevices/Device_HUDMessage_V2.Device_HUDMessage_V2_C"))
                (bind _output_get_2 (Variables|UserOptions|Basic|SetShowOnRoundStart (> _timefromroundstartoption 0))) ; Change new bool to be true
                (if _output_get_2
                  (Variables|UserOptions|Basic|SetTimeFromRoundStart (Math|Conversions|ToFloat(Integer) _timefromroundstartoption))
                  (else
                    _))
                (LocalVariables|SetOption%20Migrated true)))
            (:then_2
              ;[1 0.1681 0.9838 0.25] Migrating Sound Option from Enum to Sound Cue Soft Reference.
              (if (Utilities|Enum|NotEqual(Enum) (Variables|LegacyMigration|GetSoundOption) _soundoption)
                (Variables|UserOptions|SetPlaySound (Utilities|Select :Index (Variables|LegacyMigration|GetSoundOption) :NewEnumerator1 "/Game/Athena/Sounds/LTMs/Score/ScoreLTM_ImportantMessage_Cue.ScoreLTM_ImportantMessage_Cue" :NewEnumerator2 "/Game/Sounds/Creative/Gadgets/Search+Destroy/SearchDestroy_RoundStart_Countdown_Cue.SearchDestroy_RoundStart_Countdown_Cue" :NewEnumerator3 "/Game/Sounds/Creative/Gadgets/Search+Destroy/SearchDestroy_RoundStart_Start_Cue.SearchDestroy_RoundStart_Start_Cue" :NewEnumerator4 "/Game/Sounds/Creative/Gadgets/Search+Destroy/SearchDestroy_Message_Cue.SearchDestroy_Message_Cue"))
                (Variables|LegacyMigration|SetSoundOption _soundoption)
                (LocalVariables|SetOption%20Migrated true)))
            (:then_3
              ;[1 0.1681 0.9838 0.25] Text Style
              (if (Utilities|Enum|NotEqual(Enum) _styleoption_2 _styleoption)
                (Variables|Legacy|SetText%20Style _styleoption_2)
                (Variables|LegacyMigration|SetStyleOption _styleoption)
                (LocalVariables|SetOption%20Migrated true)))
            (:then_4
              ;[0 0.774 1 0.2] Converting old 'Message Type' to new Widget picker.
              (if (Utilities|Enum|NotEqual(Enum) (Class|GetClassDefaults "/CreativeCoreDevices/Device_HUDMessage_V2.Device_HUDMessage_V2_C") _message_priority)
                (if (Utilities|Enum|Equal(Enum) _message_priority "NewEnumerator1")
                  (Variables|UserOptions|SetHUD%20Widget "/CreativeCoreDevices/HUDMessage/Widget_HUDMessage_Exclamation.Widget_HUDMessage_Exclamation_C") ; If not critical (1), it's one that wants to use the exclamation
                  (else
                    _))
                (Variables|UserOptions|SetPriority (Utilities|Select :Index (Variables|LegacyMigration|GetMessage%20Priority) :NewEnumerator0 5 :NewEnumerator1 4 :NewEnumerator2 3))
                (Variables|UserOptions|SetPriority_Override false)
                (Variables|LegacyMigration|SetMessage%20Priority (Class|GetClassDefaults "/CreativeCoreDevices/Device_HUDMessage_V2.Device_HUDMessage_V2_C"))
                (LocalVariables|SetOption%20Migrated true)))) ; 0 - Display Time\n1 - Time from Round Start\n2 - Sound
        )
        (:then_1
          ;[1 0.1681 0.9838 0.25] If any options have migrated, refresh save data from this instance's uproperties.
          (if (LocalVariables|GetOption%20Migrated)
            (OptionData|RefreshPropertyOverrides (Variables|Default|GetToyOptionsComponent))))))))

(fn MigrateRichTextStyle
  (bind _array_element (Utilities|Array|ForEachLoopwithBreak (LocalVariables|GetSyntaxOnLegacyRichTextStyle))
    (:LoopBody
      (if (> (Utilities|String|FindSubstring (Utilities|String|ToString(Text) (Variables|UserOptions|Basic|GetMessage)) _array_element) -1)
        (Variables|LegacyMigration|SetIsUsingLegacyRichTextStyle true)
        (Variables|UserOptions|SetTextStyleSet "/CreativeCoreDevices/Athena/UI/InGame/Creative/PropertyEditor/StyleEditor/RichTextStyleSets/CreativeRichTextStyles_LegacyHUD.CreativeRichTextStyles_LegacyHUD")))))

(fn ResetOverrideVariables
  (Variables|Verse|SetMessageOverride)
  (Variables|Default|SetDisplayTimeOverride 0.0) ; <0 is permanent, 0.0 is unset, > 0 set
)

(fn SetDisplayForFiniteDuration
  (params (DesiredDisplayTime :type Float%20%28double-precision%29))
  (Networking|SwitchHasAuthority
    (:Authority
      (Variables|Default|SetDisplayTimeOverride (Math|Float|Clamp(Float) DesiredDisplayTime 0.0 6e+01)))))

(fn SetDisplayForInfiniteDuration
  ;[1 1 1] <0.f treated as permanent
  (Networking|SwitchHasAuthority
    (:Authority
      (Variables|Default|SetDisplayTimeOverride -1.0)
      (Networking|FlushNetDormancy))))

(fn SetMessageOverride
  (params (NewMessage :type Localizable%20Message%20Structure))
  (Networking|SwitchHasAuthority
    (:Authority
      (Variables|Verse|SetMessageOverride NewMessage)
      ;[1 1 1] FORT-872979: Added to still allow older version to work the same way
      (if (Variables|Default|GetShowTextOnOverride)
        (CallFunction|ShowMessage 0)))))

(fn ShouldShowTimedMessaged
  (Utilities|FlowControl|Sequence
    (:then_0
      ;[1 1 1] Only show messages on clients
      (Networking|CanExecuteCosmeticEvents
        (:False
          (return false))))
    (:then_1
      ;[1 1 1] There is still time to show the message
      (return (> (Helpers|GetTimeIntervalUntilMessageIsShown) -0.1)))))

(fn ShowQueuedMessagesOnTeamSwitch
  (params (Actor :type Actor%20Object%20Reference))
  (bind _returnvalue (Fortnite|GetFortPlayerControllerfromActor Actor))
  (bind _true (CallFunction|IsRelevantForPlayer _returnvalue))
  (if _true
    (bind _comp (CallFunction|GetPCComp _returnvalue))
    ;[0.3501 0.7657 1 0.775] Display all the queued messages. This will enqueue each message in the CRD_HudMessageComp
    (for _array_element (Variables|Default|GetMessageQueue)
      (Server|HUDMessageRequest _comp self true (Utilities|Struct|MakeHUDMessageQueueDetails (Utilities|Time|GetGameTimeinSeconds) (.InstigatorPS_13_C6228DA247AA3CC4ADA8B28C74DAD64C (Utilities|Struct|BreakHUDMessageQueueDetails _array_element)) (.MessagePayload_11_B3FB8AB54557DCEB753F2091D88BD21A (Utilities|Struct|BreakHUDMessageQueueDetails _array_element)))))))

(fn TeamSettingToActorTeam
  ;[1 0 0.09645] TODO DEVICE CONVERSION:
  ; 0 = any team?
  ; This logic may need fixing all the
  ; way into AthenaScoreAlert
  (return (select (Creative|TeamOption|IsTeamType (Variables|UserOptions|GetMessage%20Recipient) "None") 0 (Creative|TeamOption|GetLegacyTeamIndex (Variables|UserOptions|GetMessage%20Recipient)))))

(fn UpdateForActors
  (params (Relevant%20PC :type Controller%20Object%20Reference) (Add :type Boolean :default true))
  ;[0 0.9598 1 0.2] Attempts to add or remove a device from the HUDMessage Player Controller Component for the relevant players, based on function inputs. If that player doesn't yet have the component, it'll Create one for them.
  (bind _playerstates (CallFunction|GetRelevantPlayers Relevant%20PC))
  (for _array_element _playerstates
    (bind _comp (CallFunction|GetPCComp _array_element))
    (bind (_payload _last_instigator_player_state) (CallFunction|GetWidgetDetails))
    (Server|HUDMessageRequest _comp self Add (Utilities|Struct|MakeHUDMessageQueueDetails (Utilities|Time|GetGameTimeinSeconds) _last_instigator_player_state _payload)))
  (Variables|Default|SetActiveTime (Math|Float|SelectFloat (Utilities|Time|GetGameTimeinSeconds) 0.0 (Variables|GetAdd)))
  (bind (_payload_2 _last_instigator_player_state_2) (CallFunction|GetWidgetDetails))
  (if (Variables|GetAdd)
    (Utilities|Array|Add (Variables|Default|GetMessageQueue) (Utilities|Struct|MakeHUDMessageQueueDetails (Utilities|Time|GetGameTimeinSeconds) _last_instigator_player_state_2 _payload_2))
    (if (not (Variables|Default|GetIsQueueUpdating))
      (CallFunction|UpdateMessageQueue)
      (Variables|Default|SetIsQueueUpdating true))))

(fn ConstructionScript)
