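; Fortnite Creative Item Granter device: gives players or teams items that were dropped onto it outside active games.
; Trigger, timer and cycle events build grant requests, queued and processed server-side one at a time via short timer re-calls.
; Each request async-loads the item, applies grant condition and behavior, then grants, drops excess as pickups and optionally equips.
; Timed grants can accrue offline using per-player persisted timestamps to compute a multiplier.
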(event Custom|Process%20Grant%20Request%20Queue
  (bind (_pickupinstigatorhandle _itemvariantguid _weaponammooverride) (Utilities|Struct|BreakMinigameSpawnerSpawnParams (Variables|Transient|GetProcessingGrantRequestSpawnParams)))
  (bind (_pickupinstigatorhandle_2 _itemvariantguid_2 _weaponammooverride_2) (Utilities|Struct|BreakMinigameSpawnerSpawnParams (Variables|Transient|GetProcessingGrantRequestSpawnParams)))
  (bind (_returnvalue _excessitems) (CallFunction|CalculateItemNumber))
  (bind (_processinggrantrequest_controllers_38_822ad16d47dcab66ef0d7c9ce6898856 _processinggrantrequest_grantitemindex_21_b2c0fe6d40d1128d9ea9c192f324d978) (Utilities|Struct|BreakSCreativeItemGranterGrantRequest (Variables|GetProcessingGrantRequest_Controllers_38_822AD16D47DCAB66EF0D7C9CE6898856)))
  (bind (_best_valid_item_index _isvalid) (CallFunction|GetBestValidItemIndex _processinggrantrequest_grantitemindex_21_b2c0fe6d40d1128d9ea9c192f324d978))
  (bind _validitemindexes (Variables|Transient|GetValidItemIndexes))
  (bind _output (Utilities|Array|Get(aref) _validitemindexes _best_valid_item_index))
  (bind (_processinggrantrequest_controllers_38_822ad16d47dcab66ef0d7c9ce6898856_2 _processinggrantrequest_grantitemindex_21_b2c0fe6d40d1128d9ea9c192f324d978_2 _processinggrantrequest_countoverride_41_d8a66e564fea2a9fded634a7eddc274c _processinggrantrequest_executeitem_8_e7c22d144c7730e28a63608db73c7276 _processinggrantrequest_removeitem_20_8b360fe344295a37c03d84bfac65befd _processinggrantrequest_grantbehavior_32_af1cd62340c36f7d698dd59f7c7506f9) (Utilities|Struct|BreakSCreativeItemGranterGrantRequest (Variables|GetProcessingGrantRequest_Controllers_38_822AD16D47DCAB66EF0D7C9CE6898856)))
  (bind _processinggrantrequestinventoryowner (Variables|Transient|GetProcessingGrantRequestInventoryOwner))
  (bind (_processinggrantrequest_controllers_38_822ad16d47dcab66ef0d7c9ce6898856_3 _processinggrantrequest_grantitemindex_21_b2c0fe6d40d1128d9ea9c192f324d978_3 _processinggrantrequest_countoverride_41_d8a66e564fea2a9fded634a7eddc274c_2 _processinggrantrequest_executeitem_8_e7c22d144c7730e28a63608db73c7276_2) (Utilities|Struct|BreakSCreativeItemGranterGrantRequest (Variables|GetProcessingGrantRequest_Controllers_38_822AD16D47DCAB66EF0D7C9CE6898856)))
  (bind _processinggrantrequest_controllers_38_822ad16d47dcab66ef0d7c9ce6898856_4 (Utilities|Struct|BreakSCreativeItemGranterGrantRequest (Variables|GetProcessingGrantRequest_Controllers_38_822AD16D47DCAB66EF0D7C9CE6898856)))
  (bind (_processinggrantrequest_controllers_38_822ad16d47dcab66ef0d7c9ce6898856_5 _processinggrantrequest_grantitemindex_21_b2c0fe6d40d1128d9ea9c192f324d978_4 _processinggrantrequest_countoverride_41_d8a66e564fea2a9fded634a7eddc274c_3 _processinggrantrequest_executeitem_8_e7c22d144c7730e28a63608db73c7276_3 _processinggrantrequest_removeitem_20_8b360fe344295a37c03d84bfac65befd_2 _processinggrantrequest_grantbehavior_32_af1cd62340c36f7d698dd59f7c7506f9_2 _processinggrantrequest_grantcondition_34_0e6824e546464c43b5988bad6d7f9c9e) (Utilities|Struct|BreakSCreativeItemGranterGrantRequest (Variables|GetProcessingGrantRequest_Controllers_38_822AD16D47DCAB66EF0D7C9CE6898856)))
  (bind _processinggrantrequestinventoryowner_2 (Variables|Transient|GetProcessingGrantRequestInventoryOwner))
  (bind _processinggrantrequestinventoryowner_3 (Variables|Transient|GetProcessingGrantRequestInventoryOwner))
  (bind (_processinggrantrequest_controllers_38_822ad16d47dcab66ef0d7c9ce6898856_6 _processinggrantrequest_grantitemindex_21_b2c0fe6d40d1128d9ea9c192f324d978_5 _processinggrantrequest_countoverride_41_d8a66e564fea2a9fded634a7eddc274c_4 _processinggrantrequest_executeitem_8_e7c22d144c7730e28a63608db73c7276_4 _processinggrantrequest_removeitem_20_8b360fe344295a37c03d84bfac65befd_3 _processinggrantrequest_grantbehavior_32_af1cd62340c36f7d698dd59f7c7506f9_3 _processinggrantrequest_grantcondition_34_0e6824e546464c43b5988bad6d7f9c9e_2 _processinggrantrequest_lastitem_36_f36492d5444f9101a9ea0bab57a98f7e) (Utilities|Struct|BreakSCreativeItemGranterGrantRequest (Variables|GetProcessingGrantRequest_Controllers_38_822AD16D47DCAB66EF0D7C9CE6898856)))
  (Networking|SwitchHasAuthority
    (:Authority
      ;[0 1 0] Ensure the mini game is in progress
      (bind _fortminigameprogress (Variables|Default|GetFortMinigameProgress)
        (:then
          (bind _currentminigame (Variables|/Script/FortniteGame.FortMinigameLogicComponent|GetCurrentMinigame _fortminigameprogress)
            (:then
              (if (Minigame|IsinProgress _currentminigame)
                (if (Math|Boolean|ANDBoolean (Utilities|IsArrayNonEmpty (Variables|Transient|GetGrantRequestQueue)) (not (Variables|Transient|GetbProcessingGrantRequests)) (Variables|Transient|GetbHasCalledOnMinigameStarted))
                  (Variables|Transient|SetbProcessingGrantRequests true)
                  ;[1 0 0] Grab and remove the first one off the queue
                  (Variables|Transient|SetProcessingGrantRequest (Utilities|Array|Get(aref) (Variables|Transient|GetGrantRequestQueue)))
                  (Utilities|Array|RemoveIndex (Variables|Transient|GetGrantRequestQueue))
                  ;[1 0.25 0] Evaluate the best valid item index and then load the soft object pointer
                  (if (and _isvalid (Utilities|Array|IsValidIndex (Variables|Transient|GetValidItemIndexes) _best_valid_item_index))
                    (if (!= _output -1)
                      (Variables|Transient|SetProcessingGrantRequestItemIndex _output)
                      (bind _structout (Default|TryOverrideInitialWeaponAmmunition (Variables|Default|GetPickupItemList) (ItemData|MakeSpawnParams (Variables|Default|GetPickupItemList) _output (Transformation|GetActorTransform)) (Variables|UserOptions|GetInitialWeaponAmmo_Override) (Variables|UserOptions|GetInitialWeaponAmmo)))
                      (bind _output_get (Variables|Transient|SetProcessingGrantRequestSpawnParams _structout))
                      (if (Utilities|IsValidSoftObjectReference (Utilities|Struct|BreakMinigameSpawnerSpawnParams _output_get))
                        (bind _object (Utilities|AsyncLoadAsset (Utilities|Struct|BreakMinigameSpawnerSpawnParams _output_get))
                          (:Completed
                            (Utilities|FlowControl|Sequence
                              (:then_0
                                (bind _asfort_world_item_definition (Utilities|Casting|CastToFortWorldItemDefinition _object))
                                (Variables|Transient|SetProcessingGrantRequestWorldItemDefinition _asfort_world_item_definition)
                                ;[0 0 1] Loop through the players for this grant request and ensure the player controller is valid
                                (for _array_element _processinggrantrequest_controllers_38_822ad16d47dcab66ef0d7c9ce6898856_4
                                  (Variables|Transient|SetProcessingGrantRequestController 0)
                                  (Variables|Transient|SetProcessingGrantRequestInventoryOwner 0)
                                  (if (Utilities|IsValid _array_element :!class /Script/Engine.KismetSystemLibrary)
                                    (Utilities|Casting|CastToAIController _array_element
                                      (:then
                                        _)
                                      (:CastFailed
                                        (bind _playerstate (Variables|/Script/Engine.Controller|GetPlayerState _array_element)
                                          (:then
                                            (bind _asfort_player_state (Utilities|Casting|CastToFortPlayerState _playerstate))
                                            (if (Minigame|IsPlayerinMinigame (Variables|/Script/FortniteGame.FortMinigameLogicComponent|GetCurrentMinigame (Variables|Default|GetFortMinigameProgress)) _asfort_player_state false)
                                              _))
                                          (:else)))) ; Skip minigame check for AI
                                    (bind _output_get_2 (Variables|Transient|SetProcessingGrantRequestController _array_element))
                                    (bind _asfort_inventory_owner_interface (Utilities|Casting|CastToFortInventoryOwnerInterface _output_get_2))
                                    (Variables|Transient|SetProcessingGrantRequestInventoryOwner _asfort_inventory_owner_interface)
                                    ;[0.25 0 1] Check grant conditions to ensure we should give the item to the player
                                    (switch Utilities|FlowControl|Switch|SwitchonE_Device_ItemGranter_GrantCondition _processinggrantrequest_grantcondition_34_0e6824e546464c43b5988bad6d7f9c9e
                                      (:NewEnumerator0
                                        _)
                                      (:NewEnumerator1
                                        (if (not (Inventory|DoesInventoryOwnerHaveItems _processinggrantrequestinventoryowner_2 (Variables|Default|GetExcludedItemTypes)))
                                          (Variables|Default|SetbOnlyIfInitiallyEmpty true)))
                                      (:NewEnumerator2
                                        (if (> (Inventory|GetInventorySpace _processinggrantrequestinventoryowner_2) 0)
                                          _))
                                      (:NewEnumerator19
                                        (if (Variables|Default|GetbOnlyIfInitiallyEmpty)
                                          _))
                                      (:NewEnumerator20
                                        _))
                                    ;[1 1 0] Before granting the item, perform a behavior
                                    (switch Utilities|FlowControl|Switch|SwitchonE_Device_ItemGranter_GrantBehavior _processinggrantrequest_grantbehavior_32_af1cd62340c36f7d698dd59f7c7506f9
                                      (:NewEnumerator0
                                        (Inventory|ClearDroppableItems _processinggrantrequestinventoryowner))
                                      (:NewEnumerator1
                                        (Inventory|RemoveNonResourceDroppableItems _processinggrantrequestinventoryowner))
                                      (:NewEnumerator2
                                        (Inventory|ReduceDroppableResources _processinggrantrequestinventoryowner 1.0))
                                      (:NewEnumerator3
                                        _))
                                    ;[1 1 1] Plug in Multiplier check by getting the saved Date Data here.  Need to Check if player is coming back vs if the device was disabled.
                                    (if (and (Variables|UserOptions|GetbGrantOnTimer) (Variables|UserOptions|GetGrantWhileOffline))
                                      (Default|CalculateItemstoGrantOverTime)
                                      (else
                                        (Variables|Transient|SetItemGrantMultiplier 1)))
                                    ;[1 1 1] Audio
                                    (if (Variables|UserOptions|GetPlayAudio)
                                      (Creative|GameplayEffects|SimpleApplyGameplayEffectToActor (Variables|Default|GetItemGrantedGameplayEffect) (Variables|Transient|GetProcessingGrantRequestController)))
                                    ;[0.02384 0.349 0.205] Grant the item
                                    (Variables|Transient|SetCurrentAmountToGrant _returnvalue)
                                    (Variables|Transient|SetCurrentExcessAmount _excessitems)
                                    (if (> (Variables|Transient|GetCurrentAmountToGrant) 0)
                                      (Inventory|GiveItemtoInventoryOwner _processinggrantrequestinventoryowner_3 (Variables|Transient|GetProcessingGrantRequestWorldItemDefinition) _itemvariantguid_2 (Variables|Transient|GetCurrentAmountToGrant) true -1 _pickupinstigatorhandle_2 true _weaponammooverride_2)
                                      (else
                                        _))
                                    (if (> (Variables|Transient|GetCurrentExcessAmount) 0)
                                      (Inventory|SpawnPickupInWorldWithClassAndItemEntry (Creative|ItemData|CreateItemVariantEntry (Variables|Transient|GetProcessingGrantRequestWorldItemDefinition) (Variables|Transient|GetCurrentExcessAmount) -1 _itemvariantguid _pickupinstigatorhandle _weaponammooverride) "/Game/Creative/Items/Gameplay/FortCreativePickup/Fort_Pickup_Creative.Fort_Pickup_Creative_C" (Transformation|GetActorLocation (Pawn|GetControlledPawn (Variables|Transient|GetProcessingGrantRequestController))) "0.000000,0.000000,1.000000" 0 true true false "Tossed" "ItemSpawner" (Utilities|Select :Index (Variables|UserOptions|GetOwnershipOfDroppedItem) :NewEnumerator2 (Utilities|Casting|CastToFortPlayerControllerAthena (Variables|Transient|GetProcessingGrantRequestController)) :NewEnumerator1 (Utilities|Casting|CastToFortPlayerControllerAthena (Variables|Transient|GetTriggerInstigator))))
                                      (else
                                        _))
                                    (CallFunction|MaybeGiveAmmoToInventoryOwner _processinggrantrequestinventoryowner_3 (Variables|Transient|GetProcessingGrantRequestWorldItemDefinition))
                                    (CallFunction|SendwhenItemGrantedTriggerMessage (Variables|Transient|GetProcessingGrantRequestController) (+ (Variables|Transient|GetCurrentAmountToGrant) (Variables|Transient|GetCurrentExcessAmount)))
                                    ;[0.3539 0.6908 1] Try Equip Item
                                    (if (and _processinggrantrequest_executeitem_8_e7c22d144c7730e28a63608db73c7276_2 (CallFunction|IsPlayerAllowedtoEquipProcessingItem))
                                      (bind _asfort_player_controller (Utilities|Casting|CastToFortPlayerController (Variables|Transient|GetProcessingGrantRequestController))
                                        (:then
                                          (Inventory|ExecuteInventoryItemDefinition _asfort_player_controller (Variables|Transient|GetProcessingGrantRequestWorldItemDefinition) (Utilities|Struct|BreakMinigameSpawnerSpawnParams (Variables|Transient|GetProcessingGrantRequestSpawnParams)) 0.0 true true))
                                        (:CastFailed
                                          (bind _asfort_athena_aibot_controller (Utilities|Casting|CastToFortAthenaAIBotController (Variables|Transient|GetProcessingGrantRequestController)))
                                          (Class|FortAthenaAIBotController|EquipWeaponbyDefinition _asfort_athena_aibot_controller (Variables|Transient|GetProcessingGrantRequestWorldItemDefinition) true)))))))
                              (:then_1
                                _))))
                        (else
                          _))
                      ;[1 0 1] Clean up the process request and do any post grant request actions
                      (Utilities|FlowControl|Sequence
                        (:then_0
                          (if _processinggrantrequest_removeitem_20_8b360fe344295a37c03d84bfac65befd_3
                            (Utilities|Array|SetArrayElem (Variables|Transient|GetValidItemIndexes) (Variables|Transient|GetProcessingGrantRequestItemIndex) -1)))
                        (:then_1
                          (if _processinggrantrequest_lastitem_36_f36492d5444f9101a9ea0bab57a98f7e
                            (Variables|Default|SetbOnlyIfInitiallyEmpty false)))
                        (:then_2
                          (Variables|Transient|SetbProcessingGrantRequests false)
                          (Variables|Transient|SetProcessingGrantRequest))
                        (:then_3
                          (if (Utilities|IsArrayNonEmpty (Variables|Transient|GetGrantRequestQueue))
                            (bind _returnvalue_2 (Utilities|Time|SetTimerbyEvent (EventDispatchers|CreateEvent :!event Custom|Process%20Grant%20Request%20Queue) 0.01))
                            (Variables|Transient|SetProcessingGrantRequestTimerHandle _returnvalue_2))))
                      (else
                        _))
                    (else
                      _)))))
            (:else)))
        (:else)))))

(event Custom|EnableFromEvent
  (params (Instigator :type Controller%20Object%20Reference))
  ;[0.6921 0 1] Triggers & Receivers
  (Networking|SwitchHasAuthority
    (:Authority
      (if (CallFunction|IsinProgress)
        (Variables|Transient|SetbEnabled true)))))

(event EventInitializeDevice
  ;[1 1 1] Initialization
  (CallFunction|UpdateInventoryCache))

(event Custom|Reset
  ;[0 0.6807 1] Reset Grant Queue
  (CallFunction|InitializeItemCount)
  (CallFunction|UpdateInventoryCache)
  (Utilities|Array|Clear (Variables|Transient|GetGrantRequestQueue))
  (Variables|Transient|SetbProcessingGrantRequests false)
  (Variables|Transient|SetProcessingGrantRequest)
  (Utilities|Time|ClearandInvalidateTimerbyHandle (Variables|Transient|GetProcessingGrantRequestTimerHandle))
  (Variables|Transient|SetProcessingGrantRequestController 0)
  (Variables|Transient|SetProcessingGrantRequestInventoryOwner 0)
  (Variables|Transient|SetProcessingGrantRequestItemIndex -1)
  (Variables|Transient|SetProcessingGrantRequestSpawnParams)
  (Variables|Transient|SetProcessingGrantRequestWorldItemDefinition 0)
  (Variables|Transient|SetItemGrantMultiplier 1))

(event Custom|DisableFromEvent
  (params (Instigator :type Controller%20Object%20Reference))
  ;[0.6921 0 1] Triggers & Receivers
  (Networking|SwitchHasAuthority
    (:Authority
      (if (CallFunction|IsinProgress)
        (Variables|Transient|SetbEnabled false)))))

(event EventOnMinigameStarted
  ;[0 1 0.4804] Game Start
  (CallFunction|Reset)
  (Variables|Transient|SetbHasCalledOnMinigameStarted true)
  (Utilities|FlowControl|Sequence
    (:then_0
      (if (Utilities|IsArrayNonEmpty (Variables|Transient|GetPendingGrantRequestQueue))
        (Variables|SetBy-RefVar (Variables|Transient|GetGrantRequestQueue) (Variables|Transient|GetPendingGrantRequestQueue))
        (Utilities|Array|Clear (Variables|Transient|GetPendingGrantRequestQueue))
        (CallFunction|ProcessGrantRequestQueue)))
    (:then_2
      (if (Variables|UserOptions|GetbGrantOnGameStart)
        (CallFunction|GrantItem)))))

(event Custom|TimerGrantItem
  (bind _receivingplayers (Variables|UserOptions|GetReceivingPlayers))
  ;[0.6921 0 1] Triggers & Receivers
  (Networking|SwitchHasAuthority
    (:Authority
      (if (Math|Boolean|ANDBoolean (Variables|Transient|GetbEnabled) (CallFunction|IsinProgress) (> (Utilities|Array|Length (Variables|Transient|GetValidItemIndexes)) 0))
        (if (Variables|UserOptions|GetbGrantOnTimer)
          (CallFunction|StartGrantItemTimer)
          (else
            _))
        (if (Creative|TeamOption|IsTeamType _receivingplayers "TriggeringPlayer")
          (CallFunction|CreateGrantRequestForPlayers (Utilities|Array|MakeArray (Variables|Transient|GetTriggerInstigator)))
          (elif (Creative|TeamOption|IsTeamType _receivingplayers "All")
            (CallFunction|CreateGrantRequestforGroup 0 true)
            (else
              (CallFunction|CreateGrantRequestforGroup (select (Creative|TeamOption|IsTeamType _receivingplayers "TriggeringTeam") (Team|GetActorTeam (Variables|Transient|GetTriggerInstigator)) (Creative|TeamOption|ConverttoTeam _receivingplayers))))))))))

(event Custom|GrantItemFromEvent
  (params (Instigator :type Controller%20Object%20Reference))
  (bind _receivingplayers (Variables|UserOptions|GetReceivingPlayers))
  ;[0.6921 0 1] Triggers & Receivers
  (Variables|Transient|SetTriggerInstigator Instigator)
  (Variables|Default|SetEGrantProcessType "NewEnumerator0")
  (Networking|SwitchHasAuthority
    (:Authority
      (if (Math|Boolean|ANDBoolean (Variables|Transient|GetbEnabled) (CallFunction|IsinProgress) (> (Utilities|Array|Length (Variables|Transient|GetValidItemIndexes)) 0))
        (if (Variables|UserOptions|GetbGrantOnTimer)
          (CallFunction|StartGrantItemTimer)
          (else
            _))
        (if (Creative|TeamOption|IsTeamType _receivingplayers "TriggeringPlayer")
          (CallFunction|CreateGrantRequestForPlayers (Utilities|Array|MakeArray (Variables|Transient|GetTriggerInstigator)))
          (elif (Creative|TeamOption|IsTeamType _receivingplayers "All")
            (CallFunction|CreateGrantRequestforGroup 0 true)
            (else
              (CallFunction|CreateGrantRequestforGroup (select (Creative|TeamOption|IsTeamType _receivingplayers "TriggeringTeam") (Team|GetActorTeam (Variables|Transient|GetTriggerInstigator)) (Creative|TeamOption|ConverttoTeam _receivingplayers))))))))))

(event Custom|Grant%20Item
  (bind _receivingplayers (Variables|UserOptions|GetReceivingPlayers))
  ;[0.6921 0 1] Triggers & Receivers
  (Networking|SwitchHasAuthority
    (:Authority
      (if (Math|Boolean|ANDBoolean (Variables|Transient|GetbEnabled) (CallFunction|IsinProgress) (> (Utilities|Array|Length (Variables|Transient|GetValidItemIndexes)) 0))
        (if (Variables|UserOptions|GetbGrantOnTimer)
          (CallFunction|StartGrantItemTimer)
          (else
            _))
        (if (Creative|TeamOption|IsTeamType _receivingplayers "TriggeringPlayer")
          (CallFunction|CreateGrantRequestForPlayers (Utilities|Array|MakeArray (Variables|Transient|GetTriggerInstigator)))
          (elif (Creative|TeamOption|IsTeamType _receivingplayers "All")
            (CallFunction|CreateGrantRequestforGroup 0 true)
            (else
              (CallFunction|CreateGrantRequestforGroup (select (Creative|TeamOption|IsTeamType _receivingplayers "TriggeringTeam") (Team|GetActorTeam (Variables|Transient|GetTriggerInstigator)) (Creative|TeamOption|ConverttoTeam _receivingplayers))))))))))

(event Custom|TimerGrantPrevious
  ;[0.6921 0 1] Triggers & Receivers
  (Variables|Transient|SetGrantItemIndex (- (Variables|Transient|GetGrantItemIndex) 1))
  (Variables|Default|SetEGrantProcessType "NewEnumerator1")
  (switch Utilities|FlowControl|Switch|SwitchonEItemGranter_CycleBehavior (Variables|UserOptions|GetCycleBehavior)
    (:NewEnumerator0
      (Variables|Transient|SetGrantItemIndex (Math|Integer|Max(Integer) (Variables|Transient|GetGrantItemIndex))))
    (:NewEnumerator1
      (if (< (Variables|Transient|GetGrantItemIndex) 0)
        (Variables|Transient|SetGrantItemIndex (Utilities|Array|LastIndex (Variables|Transient|GetValidItemIndexes)))
        (else
          _))))
  (if (Variables|UserOptions|GetGrantOnCycle)
    (CallFunction|GrantItem)))

(event Custom|CycleToPreviousItemFromEvent
  (params (Instigator :type Controller%20Object%20Reference))
  ;[0.6921 0 1] Triggers & Receivers
  (Variables|Transient|SetTriggerInstigator Instigator)
  (Variables|Transient|SetGrantItemIndex (- (Variables|Transient|GetGrantItemIndex) 1))
  (Variables|Default|SetEGrantProcessType "NewEnumerator1")
  (switch Utilities|FlowControl|Switch|SwitchonEItemGranter_CycleBehavior (Variables|UserOptions|GetCycleBehavior)
    (:NewEnumerator0
      (Variables|Transient|SetGrantItemIndex (Math|Integer|Max(Integer) (Variables|Transient|GetGrantItemIndex))))
    (:NewEnumerator1
      (if (< (Variables|Transient|GetGrantItemIndex) 0)
        (Variables|Transient|SetGrantItemIndex (Utilities|Array|LastIndex (Variables|Transient|GetValidItemIndexes)))
        (else
          _))))
  (if (Variables|UserOptions|GetGrantOnCycle)
    (CallFunction|GrantItem)))

(event MinigameStartedDelegate(FortMinigameProgress)
  (Utilities|FlowControl|Sequence
    (:then_0
      (Variables|Transient|SetbEnabled (Variables|UserOptions|GetEnabledOnGameStart))
      (Variables|Transient|SetGrantItemIndex 0))
    (:then_1
      (CallFunction|Restock))
    (:then_2
      (bind _returnvalue (Game|ConstructObjectfromClass (Variables|Default|GetPersistenceDataItemGranterClass) self))
      (Variables|Default|SetPersistenceDataItemGranter _returnvalue))))

(event EventOnMinigameEnded
  ;[1 0.1469 0] Game End
  (CallFunction|Reset)
  (Variables|Transient|SetbHasCalledOnMinigameStarted false)
  (Utilities|Array|Clear (Variables|Transient|GetPendingGrantRequestQueue))
  (CallFunction|ClearGrantItemTimer)
  (Utilities|Array|Clear (Variables|Transient|GetPlayersGrantedItemsThisSession)))

(event Custom|TimerGrantNext
  ;[0.6921 0 1] Triggers & Receivers
  (Variables|Transient|SetGrantItemIndex (+ (Variables|Transient|GetGrantItemIndex) 1))
  (Variables|Default|SetEGrantProcessType "NewEnumerator19")
  (switch Utilities|FlowControl|Switch|SwitchonEItemGranter_CycleBehavior (Variables|UserOptions|GetCycleBehavior)
    (:NewEnumerator0
      (Variables|Transient|SetGrantItemIndex (Math|Integer|Min(Integer) (Variables|Transient|GetGrantItemIndex) (Utilities|Array|LastIndex (Variables|Transient|GetValidItemIndexes)))))
    (:NewEnumerator1
      (if (>= (Variables|Transient|GetGrantItemIndex) (Utilities|Array|Length (Variables|Transient|GetValidItemIndexes)))
        (Variables|Transient|SetGrantItemIndex 0)
        (else
          _))))
  (if (Variables|UserOptions|GetGrantOnCycle)
    (CallFunction|GrantItem)))

(event Custom|CycleToNextItemFromEvent
  (params (Instigator :type Controller%20Object%20Reference))
  ;[0.6921 0 1] Triggers & Receivers
  (Variables|Transient|SetTriggerInstigator Instigator)
  (Variables|Transient|SetGrantItemIndex (+ (Variables|Transient|GetGrantItemIndex) 1))
  (Variables|Default|SetEGrantProcessType "NewEnumerator19")
  (switch Utilities|FlowControl|Switch|SwitchonEItemGranter_CycleBehavior (Variables|UserOptions|GetCycleBehavior)
    (:NewEnumerator0
      (Variables|Transient|SetGrantItemIndex (Math|Integer|Min(Integer) (Variables|Transient|GetGrantItemIndex) (Utilities|Array|LastIndex (Variables|Transient|GetValidItemIndexes)))))
    (:NewEnumerator1
      (if (>= (Variables|Transient|GetGrantItemIndex) (Utilities|Array|Length (Variables|Transient|GetValidItemIndexes)))
        (Variables|Transient|SetGrantItemIndex 0)
        (else
          _))))
  (if (Variables|UserOptions|GetGrantOnCycle)
    (CallFunction|GrantItem)))

(event MinigameEndedDelegate(FortMinigameProgress)
  (Variables|Transient|SetbEnabled (Variables|UserOptions|GetEnabledOnGameStart))
  (Variables|Transient|SetGrantItemIndex 0))

(event EventDestroyed
  ;[1 0.08905 0] On Destroyed
  (if (Audio|Components|Audio|IsPlaying (Variables|Components|GetHolo%20Loop%20Cue))
    (Audio|Components|Audio|FadeOut (Variables|Components|GetHolo%20Loop%20Cue) 0.2 0.0 "Linear" :!class /Script/Engine.AudioComponent)))

(event Custom|TimerGrantRandom
  ;[0.6921 0 1] Triggers & Receivers
  (Variables|Transient|SetGrantItemIndex (Math|Random|RandomInteger (Utilities|Array|Length (Variables|Transient|GetValidItemIndexes))))
  (Variables|Default|SetEGrantProcessType "NewEnumerator20")
  (if (Variables|UserOptions|GetGrantOnCycle)
    (CallFunction|GrantItem)))

(event Custom|CycleToRandomItemFromEvent
  (params (Instigator :type Controller%20Object%20Reference))
  ;[0.6921 0 1] Triggers & Receivers
  (Variables|Transient|SetTriggerInstigator Instigator)
  (Variables|Transient|SetGrantItemIndex (Math|Random|RandomInteger (Utilities|Array|Length (Variables|Transient|GetValidItemIndexes))))
  (Variables|Default|SetEGrantProcessType "NewEnumerator20")
  (if (Variables|UserOptions|GetGrantOnCycle)
    (CallFunction|GrantItem)))

(event OnComponentHit(DropBox)
  (params (HitComponent :type Primitive%20Component%20Object%20Reference) (OtherActor :type Actor%20Object%20Reference) (OtherComp :type Primitive%20Component%20Object%20Reference) (NormalImpulse :type Vector :default "0, 0, 0") (Hit :type Hit%20Result%20Structure%20%28by%20ref%29))
  ;[0 0.4352 0.8594 0] Item Dropped on Device
  (Networking|SwitchHasAuthority
    (:Authority
      (if (not (CallFunction|IsinProgress))
        (bind _asfort_pickup_athena (Utilities|Casting|CastToFortPickupAthena OtherActor))
        (Utilities|IsPickupValid _asfort_pickup_athena (Variables|Default|GetInvalidPickupGameplayTagContainer)
          (:IsValid
            (if (and (Interaction|BlueprintCanInteract (Pickup|GetPawnWhoDroppedPickup _asfort_pickup_athena)) (not (Class|PickupItemListComponent|IsIteminQueue (Variables|Default|GetPickupItemList) _asfort_pickup_athena)))
              (bind _bitemadded (Class|PickupItemListComponent|AddItemtoQueue (Variables|Default|GetPickupItemList) _asfort_pickup_athena))
              (if _bitemadded
                (CallFunction|RPCMULTICASTAddItem _asfort_pickup_athena)))))))))

(event Custom|RestockItemsFromEvent
  (params (Instigator :type Controller%20Object%20Reference))
  ;[0.6921 0 1] Triggers & Receivers
  (CallFunction|Restock))

(event EventBeginPlay
  ;[0 1 0.1119 0.5] Initialization
  (|CallParentFunction)
  (Initialization|InitializeComponent (Variables|Default|GetCreative_ItemPreview_Component) (Variables|Components|GetPreviewMeshScene) (Variables|Components|GetPreviewStaticMesh) (Variables|Components|GetPreviewSkeletalMesh) (Variables|Components|GetHolo%20Loop%20Cue) (Variables|Default|GetPickupItemList) (Variables|Default|GetItemVariantNotifier) :!class /Game/Creative/Devices/Common/Components/Creative_ItemPreview_Component.Creative_ItemPreview_Component_C)
  (Initialization|Initialize :self (Variables|Default|GetCreative_UseRestriction_Component) :MinigameLogic (Variables|Default|GetFortMinigameProgress) :Team (Variables|UserOptions|GetReceivingPlayers) :Invert%20Team%20Selection false :Invert%20Class%20Selection false :Use%20Safe%20Class%20Indexing false)
  (Networking|SwitchHasAuthority
    (:Authority
      (CallFunction|RegisterFortPersistenceComponent))))

(event OnItemListUpdated(PickupItemList)
  (CallFunction|UpdateSpawnerFX)
  (CallFunction|Resave))

(event Custom|StartGrantItemTimer
  ;[1 1 1] Grant Item on Timer
  (bind _returnvalue (Utilities|Time|SetTimerbyFunctionName self "GrantItemOnTimer" (Variables|UserOptions|GetGrantTime) true))
  (Variables|Transient|SetGrantItemTimerHandle _returnvalue))

(event Custom|ClearGrantItemTimer
  ;[1 1 1] Grant Item on Timer
  (Utilities|Time|ClearandInvalidateTimerbyHandle (Variables|Transient|GetGrantItemTimerHandle)))

(event EventOnAddedToMinigame
  (params (Minigame :type Fort%20Minigame%20Object%20Reference))
  ;[0 0.4933 1 0.5] Enables/Disables drop box based on island's play state (On Play Mode Changed is bound in parent)
  (|CallParentFunction Minigame)
  (CallFunction|OnPlayModeChanged Minigame (Minigame|IsinPlayMode Minigame :!class /Script/FortniteGame.FortMinigame)))

(event Custom|On%20Play%20Mode%20Changed
  (params (Minigame :type Fort%20Minigame%20Object%20Reference) (bIsInPlayMode :type Boolean))
  ;[0 0.4933 1 0.5] Enables/Disables drop box based on island's play state (On Play Mode Changed is bound in parent)
  (Collision|SetCollisionEnabled (Variables|Components|GetDropBox) (select bIsInPlayMode "NoCollision" "QueryOnly")))

(event Custom|ClearSaveDataForPlayerFromEvent
  (params (Instigator :type Controller%20Object%20Reference))
  ;[1 1 1] Persistence
  (if (and (Utilities|IsValid Instigator :!class /Script/Engine.KismetSystemLibrary) (CallFunction|IsinProgress))
    (bind _asfort_player_state (Utilities|Casting|CastToFortPlayerState (Variables|/Script/Engine.Controller|GetPlayerState Instigator)))
    (Persistence|ClearData (Variables|Default|GetFortCreativePersistenceComponentBase) (PlayerState|GetPlayerUniqueNetID _asfort_player_state) (Variables|Default|GetSave%20Path))))

(event Custom|RPC_MULTICAST%20Add%20Item
  (params (Item :type Fort%20Pickup%20Athena%20Object%20Reference))
  (Networking|CanExecuteCosmeticEvents
    (:True
      (Class|CreativeItemPreviewComponent|ItemAddedToQueueFX (Variables|Default|GetCreative_ItemPreview_Component) Item))))

(event Custom|UpdateSpawnerFX
  ;[1 1 1] Show or hide emissive elements based on spawning being enabled
  (Networking|CanExecuteCosmeticEvents
    (:True
      (Development|SetHiddeninGame (Variables|Components|GetLightRays) (not (Math|Conversions|ToBoolean(Integer) (ItemData|GetNumItems (Variables|Default|GetPickupItemList)))) true))))

(fn CalculateItemNumber
  (bind _processinggrantrequestspawnparams (Variables|Transient|GetProcessingGrantRequestSpawnParams))
  (bind (_pickupquantity _itemvariantguid) (Utilities|Struct|BreakMinigameSpawnerSpawnParams _processinggrantrequestspawnparams))
  (bind _processinggrantrequestspawnparams_2 (Variables|Transient|GetProcessingGrantRequestSpawnParams))
  (bind (_pickupinstigatorhandle _itemvariantguid_2) (Utilities|Struct|BreakMinigameSpawnerSpawnParams _processinggrantrequestspawnparams_2))
  (bind _processinggrantrequestworlditemdefinition (Variables|Transient|GetProcessingGrantRequestWorldItemDefinition))
  (bind _returnvalue (Creative|ItemData|CreateItemVariantEntry _processinggrantrequestworlditemdefinition (LocalVariables|GetDesiredNum) -1 _itemvariantguid_2 _pickupinstigatorhandle))
  (bind _processinggrantrequestinventoryowner (Variables|Transient|GetProcessingGrantRequestInventoryOwner))
  (bind _returnvalue_2 (Inventory|GetOverflowCountIfItemWasAddedToInventory _processinggrantrequestinventoryowner _returnvalue))
  (bind _desirednum (LocalVariables|GetDesiredNum))
  (bind _returnvalue_3 (- _desirednum _returnvalue_2))
  (bind (_processinggrantrequest_controllers_38_822ad16d47dcab66ef0d7c9ce6898856 _processinggrantrequest_grantitemindex_21_b2c0fe6d40d1128d9ea9c192f324d978 _processinggrantrequest_countoverride_41_d8a66e564fea2a9fded634a7eddc274c _processinggrantrequest_executeitem_8_e7c22d144c7730e28a63608db73c7276 _processinggrantrequest_removeitem_20_8b360fe344295a37c03d84bfac65befd _processinggrantrequest_grantbehavior_32_af1cd62340c36f7d698dd59f7c7506f9 _processinggrantrequest_grantcondition_34_0e6824e546464c43b5988bad6d7f9c9e) (Utilities|Struct|BreakSCreativeItemGranterGrantRequest (Variables|GetProcessingGrantRequest_Controllers_38_822AD16D47DCAB66EF0D7C9CE6898856)))
  (bind (_processinggrantrequest_controllers_38_822ad16d47dcab66ef0d7c9ce6898856_2 _processinggrantrequest_grantitemindex_21_b2c0fe6d40d1128d9ea9c192f324d978_2 _processinggrantrequest_countoverride_41_d8a66e564fea2a9fded634a7eddc274c_2) (Utilities|Struct|BreakSCreativeItemGranterGrantRequest (Variables|GetProcessingGrantRequest_Controllers_38_822AD16D47DCAB66EF0D7C9CE6898856)))
  ;[0 1 0.8098 0.25] Grants the number of items in the device * the number of times it should have activated while the loaded player was away.
  ; If the device is set to 'only grant if owned' will only grant the number of items required to top up the player's inventory to the number specified.
  (LocalVariables|SetDesiredNum (* (Math|Integer|SelectInt _processinggrantrequest_countoverride_41_d8a66e564fea2a9fded634a7eddc274c_2 _pickupquantity (> _processinggrantrequest_countoverride_41_d8a66e564fea2a9fded634a7eddc274c_2 0)) (Math|Integer|SelectInt 1 (Variables|Transient|GetItemGrantMultiplier) (Variables|UserOptions|GetRemoveItemOnGrant))))
  (if (Utilities|Enum|Equal(Enum) _processinggrantrequest_grantcondition_34_0e6824e546464c43b5988bad6d7f9c9e "NewEnumerator20")
    (LocalVariables|SetDesiredNum (Math|Integer|Max(Integer) (- (LocalVariables|GetDesiredNum) (Inventory|GetItemQuantityonInventoryOwner (Variables|Transient|GetProcessingGrantRequestInventoryOwner) (Variables|Transient|GetProcessingGrantRequestWorldItemDefinition) _itemvariantguid))))
    (else
      _))
  (switch Utilities|FlowControl|Switch|SwitchonEItemGranter_DropItems (Variables|UserOptions|GetDropItemsAtPlayerLocation)
    (:NewEnumerator0
      (return _returnvalue_3 0))
    (:NewEnumerator1
      (return 0 (LocalVariables|GetDesiredNum)))
    (:NewEnumerator2
      (return _returnvalue_3 _returnvalue_2))))

(fn CalculateItemsToGrantOverTime
  (bind _persistencedataitemgranter (Variables|Default|GetPersistenceDataItemGranter))
  (bind _processinggrantrequestcontroller (Variables|Transient|GetProcessingGrantRequestController))
  (Utilities|Array|AddUnique (Variables|Transient|GetPlayersGrantedItemsThisSession) _processinggrantrequestcontroller)
  (bind _asfort_player_state (Utilities|Casting|CastToFortPlayerState (Variables|/Script/Engine.Controller|GetPlayerState _processinggrantrequestcontroller)))
  (LocalVariables|SetCurrentGrantedPlayerState _asfort_player_state)
  (Persistence|LoadData (Variables|Default|GetFortCreativePersistenceComponentBase) (PlayerState|GetPlayerUniqueNetID _asfort_player_state) (Variables|Default|GetSave%20Path) (Variables|Default|GetPersistenceDataItemGranter))
  (if (not (Variables|/Game/Creative/Devices/ItemGranter/Persistence/PersistenceData_ItemGranter.PersistenceData_ItemGranter_C|GetPlayerBeingTracked (Variables|Default|GetPersistenceDataItemGranter)))
    (Variables|Transient|SetItemGrantMultiplier 1)
    (else
      (Variables|Transient|SetItemGrantMultiplier (Math|Float|Truncate (/ (Math|Timespan|GetTotalSeconds (Math|DateTime|DateTime-DateTime (Math|DateTime|UTCNow) (Variables|/Game/Creative/Devices/ItemGranter/Persistence/PersistenceData_ItemGranter.PersistenceData_ItemGranter_C|GetLastTimeGrantedToPlayer (Variables|Default|GetPersistenceDataItemGranter)))) (Variables|UserOptions|GetGrantTime))))))
  (Variables|/Game/Creative/Devices/ItemGranter/Persistence/PersistenceData_ItemGranter.PersistenceData_ItemGranter_C|SetLastTimeGrantedToPlayer (Math|DateTime|UTCNow) _persistencedataitemgranter)
  (Variables|/Game/Creative/Devices/ItemGranter/Persistence/PersistenceData_ItemGranter.PersistenceData_ItemGranter_C|SetPlayerBeingTracked true _persistencedataitemgranter)
  (Persistence|SaveData (Variables|Default|GetFortCreativePersistenceComponentBase) (Variables|Default|GetPersistenceDataItemGranter) (PlayerState|GetPlayerUniqueNetID (LocalVariables|GetCurrentGrantedPlayerState)) (Variables|Default|GetSave%20Path) false))

(fn CanPlayerAutoSave
  (params (FortPlayerState :type Fort%20Player%20State%20Object%20Reference))
  (return false))

(fn CreateGrantRequestForGroup
  (params (InTeam :type Byte) (bIgnoreTeamIndex :type Boolean))
  ;[1 0.6347 0.1039] Create a grant request for all players in the minigame that are on the correct team
  (bind _fortminigameprogress (Variables|Default|GetFortMinigameProgress)
    (:then
      (bind _currentminigame (Variables|/Script/FortniteGame.FortMinigameLogicComponent|GetCurrentMinigame _fortminigameprogress)
        (:then
          (for _array_element (Minigame|GetParticipatingPlayers _currentminigame)
            (Utilities|IsValid (Fortnite|GetControllerfromActor _array_element)
              (:Is%20Valid
                (if (or (Variables|GetbIgnoreTeamIndex) (Math|Byte|Equal(Byte) (Team|GetActorTeam (Fortnite|GetControllerfromActor _array_element)) (Variables|GetInTeam)))
                  (Utilities|Array|AddUnique (LocalVariables|GetParticipatingControllers) (Fortnite|GetControllerfromActor _array_element))))))
          (CallFunction|CreateGrantRequestForPlayers (LocalVariables|GetParticipatingControllers)))
        (:else)))
    (:else)))

(fn CreateGrantRequestforPlayers
  (params (Player%20Controllers :type Array%20of%20Controller%20Object%20References))
  (bind (_best_valid_item_index _isvalid) (CallFunction|GetBestValidItemIndex (Variables|Transient|GetGrantItemIndex)))
  (Utilities|FlowControl|Sequence
    (:then_0
      (switch Utilities|FlowControl|Switch|SwitchonEItemGranter_GrantType (Variables|UserOptions|GetGrant)
        (:NewEnumerator0
          (bind _result (ArrayMacros|GetValidIndex (Variables|Default|GetItemCount_Gameplay) (Variables|Transient|GetGrantItemIndex))
            (:Valid%20Index
              (LocalVariables|SetItemCount_local _result))
            (:Invalid%20Index
              _))
          (Utilities|Array|Add (select (Variables|Transient|GetbHasCalledOnMinigameStarted) (Variables|Transient|GetGrantRequestQueue) (Variables|Transient|GetPendingGrantRequestQueue)) (Utilities|Struct|MakeSCreativeItemGranterGrantRequest (Variables|GetPlayer%20Controllers) (Variables|Transient|GetGrantItemIndex) (LocalVariables|GetItemCount_local) (Math|Boolean|ANDBoolean (Variables|UserOptions|GetbEquipGrantedItem) _isvalid (== _best_valid_item_index (Math|Integer|Min(Integer) (Variables|Transient|GetGrantItemIndex) (Utilities|Array|LastIndex (Variables|Transient|GetValidItemIndexes))))) (Variables|UserOptions|GetRemoveItemOnGrant) (Variables|UserOptions|GetOnGrantAction) (Variables|UserOptions|GetGrantCondition) true)))
        (:NewEnumerator1
          (for _index (range (+ (Utilities|Array|LastIndex (Variables|Transient|GetValidItemIndexes)) 1))
            (bind _result_2 (ArrayMacros|GetValidIndex (Variables|Default|GetItemCount_Gameplay) _index)
              (:Valid%20Index
                (LocalVariables|SetItemCount_local _result_2))
              (:Invalid%20Index
                _))
            (Utilities|Array|Add (select (Variables|Transient|GetbHasCalledOnMinigameStarted) (Variables|Transient|GetGrantRequestQueue) (Variables|Transient|GetPendingGrantRequestQueue)) (Utilities|Struct|MakeSCreativeItemGranterGrantRequest (Variables|GetPlayer%20Controllers) _index (LocalVariables|GetItemCount_local) (and (Variables|UserOptions|GetbEquipGrantedItem) (== (Math|Integer|Min(Integer) (- (Variables|UserOptions|GetItemToGrant) 1) (Utilities|Array|LastIndex (Variables|Transient|GetValidItemIndexes))) _index)) (Variables|UserOptions|GetRemoveItemOnGrant) (select (== _index 0) (Variables|UserOptions|GetOnGrantAction) "NewEnumerator3") (select (== _index 0) (Variables|UserOptions|GetGrantCondition) (Utilities|Select :Index (Variables|UserOptions|GetGrantCondition) :NewEnumerator0 "NewEnumerator0" :NewEnumerator1 "NewEnumerator19" :NewEnumerator2 "NewEnumerator2" :NewEnumerator19 "NewEnumerator19" :NewEnumerator20 "NewEnumerator20")) (== _index (Utilities|Array|LastIndex (Variables|Transient|GetValidItemIndexes)))))))))
    (:then_1
      (CallFunction|ProcessGrantRequestQueue))))

(fn GetItemIndexinRange
  (params (Raw%20Index :type Integer))
  (return (Utilities|Select :Index (Variables|UserOptions|GetCycleBehavior) :NewEnumerator0 (Math|Integer|Clamp(Integer) Raw%20Index 0 (Utilities|Array|LastIndex (Variables|Transient|GetValidItemIndexes))) :NewEnumerator1 (Math|WrappingModulo Raw%20Index (Utilities|Array|Length (Variables|Transient|GetValidItemIndexes))))))

(fn GetBestValidItemIndex
  (params (In%20Grant%20Item%20Index :type Integer))
  (for _index (range In%20Grant%20Item%20Index (+ (Utilities|Array|LastIndex (Variables|Transient|GetValidItemIndexes)) 1))
    (if (!= (Utilities|Array|Get(aref) (Variables|Transient|GetValidItemIndexes) _index) -1)
      (return _index true)))
  (return -1 false))

(fn GetItemCount
  (return (Variables|UserOptions|GetItemCount)))

(fn GetItemCountAtIndex
  (params (Index :type Integer))
  (bind (_pickuptospawn _pickupquantity) (Utilities|Struct|BreakMinigameSpawnerSpawnParams (ItemData|MakeSpawnParams (Variables|Default|GetPickupItemList) Index (struct :Location "0, 0, 0" :Rotation "0, 0, 0" :Scale "1.000000,1.000000,1.000000"))))
  (bind _result (ArrayMacros|GetValidIndex (Variables|Default|GetItemCount_Gameplay) (Variables|GetIndex))
    (:Valid%20Index
      (if (> _result 0)
        (return _result)
        (else
          _)))
    (:Invalid%20Index
      _))
  (ArrayMacros|GetValidIndex (ItemData|GetContainedItems (Variables|Default|GetPickupItemList)) Index
    (:Valid%20Index
      (return _pickupquantity))
    (:Invalid%20Index
      (return 0))))

(fn GetItemGrantCountAtIndex
  (params (Index :type Integer))
  (bind (_pickuptospawn _pickupquantity) (Utilities|Struct|BreakMinigameSpawnerSpawnParams (ItemData|MakeSpawnParams (Variables|Default|GetPickupItemList) (LocalVariables|GetCycledIndex) (struct :Location "0, 0, 0" :Rotation "0, 0, 0" :Scale "1.000000,1.000000,1.000000"))))
  ;[1 1 1] FORT-903131: Added for the new Public Verse API. Use this instead of 'GetItemCountAtIndex'
  (bind _output_get (LocalVariables|SetCycledIndex (Verse|GetItemIndexInRange Index)))
  (bind _result (ArrayMacros|GetValidIndex (Variables|Default|GetItemCount_Gameplay) _output_get)
    (:Valid%20Index
      (if (> _result 0)
        (return _result)
        (else
          _)))
    (:Invalid%20Index
      _))
  (ArrayMacros|GetValidIndex (ItemData|GetContainedItems (Variables|Default|GetPickupItemList)) _output_get
    (:Valid%20Index
      (return _pickupquantity))
    (:Invalid%20Index
      (return 0))))

(fn GetItemIndex
  (return (Variables|Transient|GetGrantItemIndex)))

(fn GetItemInventoryCountAtIndex
  (params (Agent :type Controller%20Object%20Reference) (Index :type Integer))
  ;[1 0 0.9326 0.5] EPIC INTERNAL USE ONLY  (will be superceded by itemization). Returns the number of items the player has for an item that can be granted
  (Utilities|FlowControl|Sequence
    (:then_0
      (bind _asfort_inventory_owner_interface (Utilities|Casting|CastToFortInventoryOwnerInterface Agent))
      (bind _result (ArrayMacros|GetValidIndex (Variables|CachedData|GetCachedItemVariants) Index)
        (:Valid%20Index
          (bind (_result_item _result_itemvariantguid) (Utilities|Struct|BreakItemVariantHandle _result))
          (bind _asfort_item_definition (Utilities|Casting|CastToFortItemDefinition _result_item))
          (return (Inventory|GetItemQuantityonInventoryOwner _asfort_inventory_owner_interface _asfort_item_definition _result_itemvariantguid)))))
    (:then_1
      (return 0))))

(fn GetItemNameAtIndex
  (params (Index :type Integer))
  ;[0.1386 0.02461 0.15 0.5] EPIC INTERNAL USE ONLY - Will be superceded by Itemization. Returns the name of the item as a string. Does not support Item Variants
  (bind _result (ArrayMacros|GetValidIndex (Variables|CachedData|GetCachedContainedItems) (Verse|GetItemIndexInRange Index))
    (:Valid%20Index
      (return (Utilities|String|ToString(Text) (ItemDefinition|GetDisplayName _result))))
    (:Invalid%20Index
      (return ""))))

(fn GetObjectToAutoSaveForPlayer
  (params (FortPlayerState :type Fort%20Player%20State%20Object%20Reference))
  (return 0))

(fn GetPersistenceRequiredState
  (return (Variables|UserOptions|GetGrantWhileOffline)))

(fn GetVisibilityComponentsWithCollision
  (return (Utilities|Map|MakeMap (Variables|Components|GetLightRays) "NoCollision" (Variables|Debug|GetStaticMeshComponent) "QueryAndPhysics" (Variables|Components|GetPointLight) "NoCollision" (Variables|Components|GetPreviewStaticMesh) "NoCollision" (Variables|Components|GetPreviewSkeletalMesh) "NoCollision" (Variables|Components|GetHolo%20Loop%20Cue) "NoCollision")))

(fn GrantItemIndex
  (params (Controller :type Controller%20Object%20Reference) (Index :type Integer))
  (Utilities|IsValid Controller
    (:Is%20Valid
      (Verse|GrantItemIndexToControllers (Utilities|Array|MakeArray Controller) (Verse|GetItemIndexInRange Index)))))

(fn GrantItemIndextoControllers
  (params (Controllers :type Array%20of%20Controller%20Object%20References) (Item%20Index :type Integer))
  (bind (_best_valid_item_index _isvalid) (CallFunction|GetBestValidItemIndex Item%20Index))
  (bind _result (ArrayMacros|GetValidIndex (Variables|Default|GetItemCount_Gameplay) Item%20Index)
    (:Valid%20Index
      (LocalVariables|SetItemCount_local _result))
    (:Invalid%20Index
      _))
  (Utilities|Array|Add (select (Variables|Transient|GetbHasCalledOnMinigameStarted) (Variables|Transient|GetGrantRequestQueue) (Variables|Transient|GetPendingGrantRequestQueue)) (Utilities|Struct|MakeSCreativeItemGranterGrantRequest Controllers Item%20Index (LocalVariables|GetItemCount_local) (Math|Boolean|ANDBoolean (Variables|UserOptions|GetbEquipGrantedItem) _isvalid (== _best_valid_item_index (Math|Integer|Min(Integer) (- (Variables|UserOptions|GetItemToGrant) 1) (Utilities|Array|LastIndex (Variables|Transient|GetValidItemIndexes))))) (Variables|UserOptions|GetRemoveItemOnGrant) (Variables|UserOptions|GetOnGrantAction) (Variables|UserOptions|GetGrantCondition) true))
  (CallFunction|ProcessGrantRequestQueue))

(fn GrantItemIndexToAll
  (params (Index :type Integer))
  ;[0.1178 0 1 0.2] Gifts a specific item index to all valid players. This will only apply if we're either set to gift to 'All', or we're set to 'Team Index' and the player is the same team as this device.
  (if (or (Creative|TeamOption|IsTeamType (Variables|UserOptions|GetReceivingPlayers)) (Creative|TeamOption|IsTeamType (Variables|UserOptions|GetReceivingPlayers) "All"))
    (bind _fortminigameprogress (Variables|Default|GetFortMinigameProgress)
      (:then
        (bind _currentminigame (Variables|/Script/FortniteGame.FortMinigameLogicComponent|GetCurrentMinigame _fortminigameprogress)
          (:then
            (for _array_element (Minigame|GetParticipatingPlayers _currentminigame)
              (Utilities|IsValid _array_element
                (:Is%20Valid
                  (if (or (and (== (Math|Byte|MakeLiteralByte (Creative|TeamOption|ConverttoTeam (Variables|UserOptions|GetReceivingPlayers))) (Team|GetActorTeam _array_element)) (Creative|TeamOption|IsTeamType (Variables|UserOptions|GetReceivingPlayers))) (Creative|TeamOption|IsTeamType (Variables|UserOptions|GetReceivingPlayers) "All"))
                    (Utilities|Array|Add (LocalVariables|GetLocalControllers) (Fortnite|GetControllerfromActor _array_element))))))
            (Verse|GrantItemIndexToControllers (LocalVariables|GetLocalControllers) (Verse|GetItemIndexInRange Index)))
          (:else)))
      (:else))))

(fn GrantItemOnTimer
  (bind _persistencedataitemgranter (Variables|Default|GetPersistenceDataItemGranter))
  (if (CallFunction|IsinProgress)
    (if (Variables|Transient|GetbEnabled)
      (switch Utilities|FlowControl|Switch|SwitchonE_Device_ItemGranterGrantProcessType (Variables|Default|GetEGrantProcessType)
        (:NewEnumerator0
          (CallFunction|TimerGrantItem))
        (:NewEnumerator1
          (CallFunction|TimerGrantPrevious))
        (:NewEnumerator19
          (CallFunction|TimerGrantNext))
        (:NewEnumerator20
          (CallFunction|TimerGrantRandom)))
      (else
        (CallFunction|ClearGrantItemTimer)
        (if (Variables|UserOptions|GetGrantWhileOffline)
          (for _array_element (Variables|Transient|GetPlayersGrantedItemsThisSession)
            (bind _asfort_player_state (Utilities|Casting|CastToFortPlayerState (Variables|/Script/Engine.Controller|GetPlayerState _array_element)))
            (if (Minigame|IsPlayerinOurVolume (Minigame|GetMinigame self :!class /Script/FortniteGame.MinigameBlueprintLibrary) _asfort_player_state)
              (Variables|/Game/Creative/Devices/ItemGranter/Persistence/PersistenceData_ItemGranter.PersistenceData_ItemGranter_C|SetPlayerBeingTracked false _persistencedataitemgranter)
              (Persistence|SaveData (Variables|Default|GetFortCreativePersistenceComponentBase) _persistencedataitemgranter (PlayerState|GetPlayerUniqueNetID _asfort_player_state) (Variables|Default|GetSave%20Path) false))))))))

(fn InitializeItemCount
  ;[0.9243 0 0.9566 0.5] Each item can have an overridable count that can be different per item. If we override it, populate the item count array with the Item Count option for each item (this is so that verse can override them selectively).
  (if (Variables|UserOptions|GetItemCount_Override)
    (for _index (range (ItemData|GetNumItems (Variables|Default|GetPickupItemList)))
      (Utilities|Array|SetArrayElem (Variables|Default|GetItemCount_Gameplay) _index (Variables|UserOptions|GetItemCount) true))
    (else
      (Variables|Default|SetItemCount_Gameplay (Utilities|Array|MakeArray)) ; Do Not use Item Count
    )))

(fn IsInProgress
  (bind _fortminigameprogress (Variables|Default|GetFortMinigameProgress)
    (:then
      (bind _currentminigame (Variables|/Script/FortniteGame.FortMinigameLogicComponent|GetCurrentMinigame _fortminigameprogress)
        (:then
          (return (Minigame|IsinProgress _currentminigame)))
        (:else)))
    (:else)))

(fn IsPlayerAllowedToEquipProcessingItem
  (bind _returnvalue (Fortnite|Athena|GameState|GetGameStateAthena))
  ;[1 1 1] If not a trap, this function currently always returns true
  (if (Utilities|Enum|Equal(Enum) (Item|GetItemType (Variables|Transient|GetProcessingGrantRequestWorldItemDefinition)) "Trap")
    (if (Utilities|IsValid _returnvalue :!class /Script/Engine.KismetSystemLibrary)
      (return (not (Class|FortGameStateAthena|ShouldBlockTrapPlacement _returnvalue (Pawn|GetControlledPawn (Variables|Transient|GetProcessingGrantRequestController)))))
      (else
        (return true)))
    (else
      (return true))))

(fn IsVisibleDuringMinigame
  (return false))

(fn MaybeGiveAmmotoInventoryOwner
  (params (InventoryOwner :type Fort%20Inventory%20Owner%20Interface%20Interface) (WorldItemDefinition :type Fort%20World%20Item%20Definition%20Object%20Reference))
  (HelperFunctions|GiveAmmotoInventoryOwner InventoryOwner WorldItemDefinition -1 (Variables|Default|GetExtra%20Ammo%20Amounts%20Map) (Variables|Default|GetDefaultExtraAmmoAmount) (Variables|UserOptions|GetSpareWeaponAmmo_Override) (Variables|UserOptions|GetSpareWeaponAmmo)))

(fn MigrateLegacyUserOptions
  (bind _grantitemtimer (Class|GetClassDefaults "/CreativeCoreDevices/Device_ItemGranter_V2.Device_ItemGranter_V2_C"))
  (bind _bdropitemsatplayerlocation (Class|GetClassDefaults "/CreativeCoreDevices/Device_ItemGranter_V2.Device_ItemGranter_V2_C"))
  (bind _bdropitemsatplayerlocationifinventoryisfull (Class|GetClassDefaults "/CreativeCoreDevices/Device_ItemGranter_V2.Device_ItemGranter_V2_C"))
  (bind _equipitemongrant (Class|GetClassDefaults "/CreativeCoreDevices/Device_ItemGranter_V2.Device_ItemGranter_V2_C"))
  (bind _giveextraammo (Class|GetClassDefaults "/CreativeCoreDevices/Device_ItemGranter_V2.Device_ItemGranter_V2_C"))
  (bind _equipitemongrant_2 (Variables|Legacy|GetEquipItemOnGrant))
  (bind _grantitemtimer_2 (Variables|Legacy|GetGrantItemTimer))
  (|CallParentFunction)
  (Utilities|FlowControl|Sequence
    (:then_0
      (Utilities|FlowControl|Sequence
        (:then_0
          ;[0 1 0.5107 0.25] Migrate Drop Items at Player Location
          (if (!= _bdropitemsatplayerlocation (Variables|Legacy|GetbDropItemsAtPlayerLocation))
            (Variables|UserOptions|SetDropItemsAtPlayerLocation "NewEnumerator1")
            (Variables|Legacy|SetbDropItemsAtPlayerLocation _bdropitemsatplayerlocation)
            (LocalVariables|SetMigrate%20Options true)))
        (:then_1
          ;[0 1 0.5107 0.25] Migrate Drop Items if Inventory is Full only if Drop Items at Player Location is not true.
          (if (!= _bdropitemsatplayerlocationifinventoryisfull (Variables|Legacy|GetbDropItemsAtPlayerLocationIfInventoryIsFull))
            (Variables|Legacy|SetbDropItemsAtPlayerLocationIfInventoryIsFull _bdropitemsatplayerlocationifinventoryisfull)
            (if (not (Utilities|Enum|Equal(Enum) (Variables|UserOptions|GetDropItemsAtPlayerLocation) "NewEnumerator1"))
              (Variables|UserOptions|SetDropItemsAtPlayerLocation "NewEnumerator0")
              (LocalVariables|SetMigrate%20Options true))))
        (:then_2
          ;[0 1 0.5107 0.25] Migrate Equip Item on Grant
          (if (!= _equipitemongrant _equipitemongrant_2)
            (Variables|UserOptions|SetbEquipGrantedItem true)
            (Variables|UserOptions|SetItemToGrant (+ _equipitemongrant_2 1))
            (Variables|Legacy|SetEquipItemOnGrant _equipitemongrant)
            (LocalVariables|SetMigrate%20Options true)))
        (:then_3
          ;[0 1 0.5107 0.25] Migrate Grant Item Timer to Split into Bool + Float
          (if (!= _grantitemtimer _grantitemtimer_2)
            (Variables|UserOptions|SetbGrantOnTimer true)
            (Variables|UserOptions|SetGrantTime _grantitemtimer_2)
            (Variables|Legacy|SetGrantItemTimer _grantitemtimer)
            (LocalVariables|SetMigrate%20Options true)))
        (:then_4
          ;[0 1 0.5107 0.25] Migrate Give Extra Ammo
          (if (!= _giveextraammo (Variables|Legacy|GetGiveExtraAmmo))
            (Variables|UserOptions|SetSpareWeaponAmmo_Override (not (Variables|Legacy|GetGiveExtraAmmo)))
            (Variables|UserOptions|SetSpareWeaponAmmo 0)
            (Variables|Legacy|SetGiveExtraAmmo _giveextraammo)
            (LocalVariables|SetMigrate%20Options true)))))
    (:then_1
      ;[0 1 0.7765 0.25] If options were migrated, refresh property overrides
      (if (LocalVariables|GetMigrate%20Options)
        (OptionData|RefreshPropertyOverrides (Variables|Default|GetToyOptionsComponent))))))

(fn RegisterFortPersistenceComponent
  (Persistence|InitializePersistenceComponent (Variables|Default|GetFortCreativePersistenceComponentBase) (Variables|Default|GetFortMinigameProgress) true))

(fn Restock
  (Utilities|Array|Clear (Variables|Transient|GetValidItemIndexes))
  (for _index (range (ItemData|GetNumItems (Variables|Default|GetPickupItemList)))
    (Utilities|Array|AddUnique (Variables|Transient|GetValidItemIndexes) _index)))

(fn SendWhenItemGrantedTriggerMessage
  (params (Trigger%20Instigator :type Controller%20Object%20Reference) (Count :type Integer))
  (GameplayEvent|BroadcastEvent (Variables|UserOptions|Events|GetOnItemGranted) 0 (Utilities|Struct|MakeGameplayEventItemGranted Trigger%20Instigator))
  (GameplayEvent|BroadcastEvent (Variables|UserOptions|Events|GetOnItemGrantedWithCount) 0 (Utilities|Struct|MakeGameplayEventItemGrantedwithCount Trigger%20Instigator Count)))

(fn SetItemCount
  (params (ItemCount :type Integer))
  ;[1 0 0.8616 0.5] Verse. Override the item count for all items in the granter.
  (if (> ItemCount 0)
    (for _index (range (ItemData|GetNumItems (Variables|Default|GetPickupItemList)))
      (Utilities|Array|SetArrayElem (Variables|Default|GetItemCount_Gameplay) _index ItemCount true))))

(fn SetItemCountAtIndex
  (params (ItemCount :type Integer) (Index :type Integer))
  ;[0.8145 0 1 0.5] Verse - Override the item count at the specific item index.
  (if (> ItemCount 0)
    (Utilities|Array|SetArrayElem (Variables|Default|GetItemCount_Gameplay) (Verse|GetItemIndexInRange Index) ItemCount true)))

(fn SetNextItemIndex
  (params (Index :type Integer))
  (Networking|SwitchHasAuthority
    (:Authority
      (if (Utilities|Array|IsValidIndex (Variables|Transient|GetValidItemIndexes) Index)
        (Variables|Transient|SetGrantItemIndex Index)))))

(fn UnRegisterFortPersistenceComponent
  (params (FortMinigame :type Fort%20Minigame%20Object%20Reference))
  (bind _volume (Variables|/Script/FortniteGame.FortMinigame|GetVolume FortMinigame))
  (bind _returnvalue (Actor|GetComponentbyClass _volume "/Script/FortniteGame.FortVolumePersistenceOptions"))
  (Utilities|IsValid _returnvalue
    (:Is%20Valid
      (Class|FortVolumePersistenceOptions|UnregistryFortCreativePersistenceComponent _returnvalue 0))))

(fn UpdateInventoryCache
  (Variables|CachedData|SetCachedItemVariants (ItemData|GetContainedItemVariants (Variables|Default|GetPickupItemList)))
  (Variables|CachedData|SetCachedContainedItems (ItemData|GetContainedItems (Variables|Default|GetPickupItemList))))

(fn ConstructionScript
  (|CallParentFunction))
