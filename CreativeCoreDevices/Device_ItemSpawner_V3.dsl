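; Fortnite Creative Item Spawner device (V3) that spawns item pickups through a Minigame Spawner component and shows previews of its contents.
; User options configure the spawner's timing, scale, costs and pickup physics, with tilt checked for gravity tossing.
; Cycling to the next item runs on authority during an active minigame, replacing current pickups once spawning has begun.
; Container updates refresh previews and flag linked secondary weapons via a loose gameplay tag.
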(event EditorPreview|EventSetupForEditorPreview
  (bind _previewstaticmesh (Variables|Default|GetPreviewStaticMesh))
  ;[1 0.9002 0 0.25] Show the first item in the spawner during UEFN preview
  (|CallParentFunction)
  (Transformation|SetWorldScale3D _previewstaticmesh (Math|Conversions|ToVector(Float) (Variables|UserOptions|GetItem%20Scale)))
  (Class|CreativeItemPreviewComponent|EditorDisplayPreview (Variables|Default|GetCreative_ItemPreview_Component) _previewstaticmesh))

(event EventBeginPlay
  ;[0 1 0.2776 0.25] Initialization
  (Components|Tick|SetComponentTickEnabled (Variables|Default|GetPreviewSkeletalMesh) false)
  (|CallParentFunction))

(event Custom|CycleToNextItem
  (params (Instigator :type Controller%20Object%20Reference))
  (Networking|SwitchHasAuthority
    (:Authority
      (if (not (Variables|Default|GetbDisabledSpawning))
        (bind _binprogress (CallFunction|IsMinigameinProgress))
        (if _binprogress
          (if (not (Class|MinigameSpawnerComponent|HasBegunSpawning (Variables|Default|GetMinigame_Spawner_Component)))
            (Utilities|Time|ClearandInvalidateTimerbyHandle (Variables|Default|GetMinigameStartSpawnTimer))
            (Class|MinigameSpawnerComponent|SpawnRemaining (Variables|Default|GetMinigame_Spawner_Component) (not (Variables|Default|GetbDisabledSpawning)))
            (else
              (Class|MinigameSpawnerComponent|DestroyPickups (Variables|Default|GetMinigame_Spawner_Component))
              (Class|MinigameSpawnerComponent|SpawnNext (Variables|Default|GetMinigame_Spawner_Component)))))))))

(event Custom|SpawnItem
  (params (Instigator :type Controller%20Object%20Reference))
  (CallFunction|OnSpawnItemReceived Instigator))

(event OnAnyPropertyChanged(ToyOptionsComponent)
  (Variables|/Game/Athena/Items/Traps/MinigameSpawner/Minigame_Spawner_Component.Minigame_Spawner_Component_C|SetbShouldTossPickups (or (CallFunction|CanPickupbeTossedbyGravity 3e+01) (Utilities|Enum|Equal(Enum) (Variables|UserOptions|GetInitial%20Movement%20of%20Item) "NewEnumerator0")) (Variables|Default|GetMinigame_Spawner_Component))
  (Variables|/Game/Athena/Items/Traps/MinigameSpawner/Minigame_Spawner_Component.Minigame_Spawner_Component_C|SetbShouldApplyGravityToPickups (or (Utilities|Enum|Equal(Enum) (Variables|UserOptions|GetInitial%20Movement%20of%20Item) "NewEnumerator0") (Utilities|Enum|Equal(Enum) (Variables|UserOptions|GetInitial%20Movement%20of%20Item) "NewEnumerator1")) (Variables|Default|GetMinigame_Spawner_Component))
  (Transformation|SetRelativeLocation (Variables|Default|GetRotationSceneComponent) (CallFunction|GetItemPreviewLocationOffset) false false)
  (Class|CreativeItemPreviewComponent|UpdatePreviews (Variables|Default|GetCreative_ItemPreview_Component)))

(event OnPropertiesLoaded(ToyOptionsComponent)
  (Variables|/Game/Athena/Items/Traps/MinigameSpawner/Minigame_Spawner_Component.Minigame_Spawner_Component_C|SetbShouldTossPickups (or (CallFunction|CanPickupbeTossedbyGravity 3e+01) (Utilities|Enum|Equal(Enum) (Variables|UserOptions|GetInitial%20Movement%20of%20Item) "NewEnumerator0")) (Variables|Default|GetMinigame_Spawner_Component))
  (Variables|/Game/Athena/Items/Traps/MinigameSpawner/Minigame_Spawner_Component.Minigame_Spawner_Component_C|SetbShouldApplyGravityToPickups (or (Utilities|Enum|Equal(Enum) (Variables|UserOptions|GetInitial%20Movement%20of%20Item) "NewEnumerator0") (Utilities|Enum|Equal(Enum) (Variables|UserOptions|GetInitial%20Movement%20of%20Item) "NewEnumerator1")) (Variables|Default|GetMinigame_Spawner_Component))
  (Transformation|SetRelativeLocation (Variables|Default|GetRotationSceneComponent) (CallFunction|GetItemPreviewLocationOffset) false false)
  (Class|CreativeItemPreviewComponent|UpdatePreviews (Variables|Default|GetCreative_ItemPreview_Component)))

(event OnItemContainerUpdated(Minigame_Spawner_Component)
  ; If the spawner is rotated far enough then we need to see if gravity should 'toss" the item
  (CallFunction|UpdateItemPreviewList)
  (if (Variables|Default|GetbIgnoreSpawnParamsBool)
    (CallFunction|SetLinkedWeaponBool)
    (Variables|Default|SetbIgnoreSpawnParamsBool false)
    (else
      (CallFunction|UpdateLinkedWeaponTag)))
  (Class|CreativeItemPreviewComponent|UpdatePreviews (Variables|Default|GetCreative_ItemPreview_Component)))

(event Custom|SetLinkedWeaponBool
  (for _array_element (Variables|/Script/FortniteGame.FortMinigameSpawnerComponent|GetToSpawnList (Variables|Default|GetMinigame_Spawner_Component))
    (bind _object (Utilities|AsyncLoadAsset (Utilities|Struct|BreakMinigameSpawnerSpawnParams _array_element))
      (:Completed
        (bind _asfort_weapon_item_definition (Utilities|Casting|CastToFortWeaponItemDefinition _object))
        (Utilities|IsValid (WeaponItemDefinition|GetLinkedItemDefinition _asfort_weapon_item_definition)
          (:Is%20Valid
            (Utilities|FlowControl|Sequence
              (:then_0
                (Utilities|FlowControl|DoOnce false
                  (:Completed
                    (Ability|GameplayEffect|AddLooseGameplayTags self (GameplayTags|MakeLiteralGameplayTagContainer "(GameplayTags=((TagName=\"Creative.Device.ItemSpawner.SecondaryWeapon\")))")))))
              (:then_1
                (CallFunction|UpdateSpawnListLinkedWeaponBool _object)))))))))

(fn ActivateFromOtherDevice
  (params (Instigator :type Controller%20Object%20Reference) (Requester :type ERequesterType%20Enum :default "NewEnumerator0"))
  (return false))

(fn CanInteract
  (params (InteractingPawn :type Fort%20Pawn%20Object%20Reference) (InteractionBeingAttempted :type EInteractionBeingAttempted%20Enum :default "FirstInteraction") (InteractionType :type TInteractionType%20Enum :default "IT_NoInteraction"))
  (return false))

(fn CanPickupBeTossedByGravity
  (params (AcceptableAngleDevianceDegrees :type Float%20%28double-precision%29))
  ; Compares the upward angle difference of the spawner with the Acceptable Angle to determine if the Spawner is rotated enough for Gravity to effect the item
  (return (and (or (<= (Math|Vector|DotProduct (Math|Vector|GetUpVector "0, 0, 0") (Transformation|GetActorUpVector)) (Math|Trig|Acos(Degrees) (Variables|GetAcceptableAngleDevianceDegrees))) (not (Variables|UserOptions|GetBase%20Visible%20During%20Game))) (Utilities|Enum|Equal(Enum) (Variables|UserOptions|GetInitial%20Movement%20of%20Item) "NewEnumerator1"))))

(fn GetDeviceColor
  (return 0))

(fn GetInstigatingController
  (return 0))

(fn GetCreativeActorBounds
  (bind (_origin _boxextent) (Collision|GetComponentBounds (Variables|Debug|GetStaticMeshComponent)))
  (return true _boxextent))

(fn GetCreativeActorOrigin
  (return true "0, 0, 0"))

(fn GetCreativeActorRelativeTransform
  (return false 0))

(fn GetGridAxisGuideAttachmentPoint
  (return false false "0, 0, 0"))

(fn GetItemPreviewLocationOffset
  (bind (_returnvalue_origin _returnvalue_boxextent) (Math|BoxSphereBounds|BreakBoxSphereBounds (StaticMesh|GetBounds (Variables|/Script/Engine.StaticMeshComponent|GetStaticMesh (Variables|Debug|GetStaticMeshComponent)))))
  (bind (_returnvalue_origin_x _returnvalue_origin_y _returnvalue_origin_z) (Math|Vector|BreakVector _returnvalue_origin))
  (bind (_returnvalue_boxextent_x _returnvalue_boxextent_y _returnvalue_boxextent_z) (Math|Vector|BreakVector _returnvalue_boxextent))
  (return 0.0 0.0 (Math|Float|SelectFloat (Utilities|Operators|Add (* (Variables|UserOptions|GetItem%20Scale) (Class|FortPickup|GetMovementCapsuleHalfHeight)) _returnvalue_origin_z _returnvalue_boxextent_z) 1e+02 (Utilities|Enum|Equal(Enum) (Variables|UserOptions|GetInitial%20Movement%20of%20Item) "NewEnumerator1"))))

(fn GetSpawnedPickups
  (return 0))

(fn IsEnabled
  (return (not (Variables|Default|GetbDisabledSpawning))))

(fn IsHoldingItemForPawn
  (params (Item :type Item%20Variant%20Handle%20Structure) (Pawn :type Pawn%20Object%20Reference))
  (return false))

(fn OverrideWeaponAmmo
  (|CallParentFunction)
  (Default|TryOverrideSecondaryWeaponAmmoonSpawnParams (Variables|Default|GetMinigame_Spawner_Component) (Variables|UserOptions|GetInitialSecondaryWeaponAmmo_Override) (Variables|UserOptions|GetInitialSecondaryWeaponAmmo)))

(fn ResaveActor
  (bind _minigame_spawner_component (Variables|Default|GetMinigame_Spawner_Component))
  (Variables|/Game/Athena/Items/Traps/MinigameSpawner/Minigame_Spawner_Component.Minigame_Spawner_Component_C|SetDelayBetweenSpawns (Math|Float|SelectFloat (Math|Float|SelectFloat 0.001 (Variables|UserOptions|Basic|GetTime%20Between%20Spawns) (<= (Variables|UserOptions|Basic|GetTime%20Between%20Spawns) 0)) -1.0 (Variables|UserOptions|Basic|GetRespawn%20Item%20On%20Timer)) _minigame_spawner_component)
  (Variables|/Game/Athena/Items/Traps/MinigameSpawner/Minigame_Spawner_Component.Minigame_Spawner_Component_C|SetIsPickupOnTouch (Variables|UserOptions|GetRun%20Over%20Pickup) _minigame_spawner_component)
  (Variables|/Game/Athena/Items/Traps/MinigameSpawner/Minigame_Spawner_Component.Minigame_Spawner_Component_C|SetItemSpawnScale (Math|Vector|MakeVector (Variables|UserOptions|GetItem%20Scale) (Variables|UserOptions|GetItem%20Scale) (Variables|UserOptions|GetItem%20Scale)) _minigame_spawner_component)
  (Variables|/Game/Athena/Items/Traps/MinigameSpawner/Minigame_Spawner_Component.Minigame_Spawner_Component_C|SetItemWoodCost (Variables|UserOptions|GetWood%20Cost%20of%20Item) _minigame_spawner_component)
  (Variables|/Game/Athena/Items/Traps/MinigameSpawner/Minigame_Spawner_Component.Minigame_Spawner_Component_C|SetItemMetalCost (Variables|UserOptions|GetMetal%20Cost%20of%20Item) _minigame_spawner_component)
  (Variables|/Game/Athena/Items/Traps/MinigameSpawner/Minigame_Spawner_Component.Minigame_Spawner_Component_C|SetItemStoneCost (Variables|UserOptions|GetStone%20Cost%20of%20Item) _minigame_spawner_component)
  (Variables|/Game/Athena/Items/Traps/MinigameSpawner/Minigame_Spawner_Component.Minigame_Spawner_Component_C|SetItemGoldCost (Variables|UserOptions|GetGold%20Cost%20of%20Item) _minigame_spawner_component)
  (Variables|/Game/Athena/Items/Traps/MinigameSpawner/Minigame_Spawner_Component.Minigame_Spawner_Component_C|SetNumSimultaneous (select (Variables|UserOptions|GetContinuously%20Spawn%20Items) 500 1) _minigame_spawner_component)
  (Variables|/Game/Athena/Items/Traps/MinigameSpawner/Minigame_Spawner_Component.Minigame_Spawner_Component_C|SetSpareLinkedWeaponAmmo (Variables|UserOptions|GetSpareSecondaryWeaponAmmo) _minigame_spawner_component)
  (Variables|/Game/Athena/Items/Traps/MinigameSpawner/Minigame_Spawner_Component.Minigame_Spawner_Component_C|SetSpareLinkedWeaponAmmo_Override (Variables|UserOptions|GetSpareSecondaryWeaponAmmo_Override) _minigame_spawner_component)
  (|CallParentFunction))

(fn SetTriggerCollision
  (params (On :type Boolean))
  (return false))

(fn UpdateLinkedWeaponTag
  (bind _returnvalue (GameplayTags|MakeLiteralGameplayTagContainer "(GameplayTags=((TagName=\"Creative.Device.ItemSpawner.SecondaryWeapon\")))"))
  (bind _self self)
  (Ability|GameplayEffect|RemoveLooseGameplayTags _self _returnvalue)
  (bind _array_element (Utilities|Array|ForEachLoopwithBreak (Variables|/Script/FortniteGame.FortMinigameSpawnerComponent|GetToSpawnList (Variables|Default|GetMinigame_Spawner_Component)))
    (:LoopBody
      (if (Utilities|Struct|BreakMinigameSpawnerSpawnParams _array_element)
        (Ability|GameplayEffect|AddLooseGameplayTags _self _returnvalue)
        (return)))))

(fn UpdateSpawnListLinkedWeaponBool
  (params (Object :type Object%20Reference))
  (bind _index (LocalVariables|GetIndex))
  (bind (_array_element _array_index) (Utilities|Array|ForEachLoopwithBreak (Variables|/Script/FortniteGame.FortMinigameSpawnerComponent|GetToSpawnList (Variables|Default|GetMinigame_Spawner_Component)))
    (:LoopBody
      (if (Utilities|String|EqualExactly(String) (Utilities|BreakSoftObjectPath (Utilities|ToSoftObjectPath (Utilities|Struct|BreakMinigameSpawnerSpawnParams _array_element))) (Utilities|BreakSoftObjectPath (Utilities|GetSoftObjectPath Object)))
        (LocalVariables|SetIndex _array_index)))
    (:Completed
      _))
  (if (>= _index 0)
    (Utilities|Struct|SetmembersinMinigameSpawnerSpawnParams (Utilities|Array|Get(aref) (Variables|/Script/FortniteGame.FortMinigameSpawnerComponent|GetToSpawnList (Variables|Default|GetMinigame_Spawner_Component)) _index) true)))

(fn ConstructionScript
  (|CallParentFunction))
