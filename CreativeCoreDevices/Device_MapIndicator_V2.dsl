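; Fortnite Creative Map Indicator device that shows a configurable icon and text marker on the map and can toggle an objective trail.
; On initialization it configures the marker with team/class restrictions, and its enabled component, driven by game start/end and Enable/Disable events, controls marker visibility.
; Player controllers receive a component registering the marker at game start or on join-in-progress.
; Legacy color and team/class options are migrated.
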(event EventInitializeDevice
  ;[1 1 1] Defaults
  (|CallParentFunction)
  (Initialization|Initialize (Variables|Default|GetCreative_UseRestriction_Component) (Variables|Default|GetFortMinigameProgress) (Variables|Default|GetEnabledComponent) 0 (Variables|UserOptions|GetAssignedTeam) (Variables|UserOptions|GetInvertTeam) (Variables|UserOptions|GetAssignedClass) (Variables|UserOptions|GetInvertClass) false)
  (Class|CreativeMapMarkerComponent|SetupMapMarker (Variables|Default|GetCreativeMapMarkerComponent) (Creative|UI|IconPicker|GetIconLibraryTextureWithEmptyTexture (Variables|UserOptions|Basic|GetIcon)) (Variables|UserOptions|Basic|GetIconColor) (Variables|UserOptions|GetShowOnWhichMap) true (Variables|UserOptions|GetAssignedTeam) (CallFunction|GetInitialMarkerVisibility) (Variables|UserOptions|Basic|GetText) (Variables|UserOptions|Basic|GetTextColor) (Variables|Default|GetCreative_UseRestriction_Component) "IconTexture" (Variables|UserOptions|GetCustomIconMaterial) (Variables|UserOptions|Basic|GetIconScale)))

(event EventOnMinigameStarted
  ;[1 1 1] Defaults
  (|CallParentFunction)
  (Class|CreativeEnabledComponent|SetEnabled (Variables|Default|GetEnabledComponent) (Variables|UserOptions|Basic|GetEnabledOnGameStart))
  (CallFunction|AddAllControllers))

(event EventOnMinigameEnded
  ;[1 1 1] Defaults
  (|CallParentFunction)
  (Class|CreativeEnabledComponent|SetEnabled (Variables|Default|GetEnabledComponent) true))

(event Custom|Enable%20Event
  (params (Instigator :type Controller%20Object%20Reference))
  ;[1 1 1] Triggers & Receivers
  (Class|CreativeEnabledComponent|SetEnabled (Variables|Default|GetEnabledComponent) true))

(event Custom|Disable%20Event
  (params (Instigator :type Controller%20Object%20Reference))
  ;[1 1 1] Triggers & Receivers
  (Class|CreativeEnabledComponent|SetEnabled (Variables|Default|GetEnabledComponent) false))

(event OnEnabledStateChanged(EnabledComponent)
  (params (Enabled :type Boolean))
  ;[1 1 1] Triggers & Receivers
  (MiniMapData|SetMiniMapIndicatorIsVisible (Variables|Default|GetCreativeMapMarkerComponent) Enabled))

(event Custom|Activate%20Objective%20Trail
  (params (Instigator :type Controller%20Object%20Reference))
  ;[1 1 1] Objective Trail Component
  (Class|CreativeObjectiveTrailComponent|ToggleObjectiveTrail (Variables|Default|GetCreative_ObjectiveTrail_Component) true self Instigator (Creative|TeamOption|GetLegacyTeamIndex (Variables|UserOptions|GetAssignedTeam)) (Variables|UserOptions|GetShowObjectivePulseToFriendlyPlayers) false (Variables|UserOptions|GetShowObjectivePulseToInstigatorOnly) false 2.0 (Variables|Default|GetCreative_UseRestriction_Component)))

(event Custom|Deactivate%20Objective%20Trail
  (params (Instigator :type Controller%20Object%20Reference))
  ;[1 1 1] Objective Trail Component
  (Class|CreativeObjectiveTrailComponent|ToggleObjectiveTrail (Variables|Default|GetCreative_ObjectiveTrail_Component) false self Instigator (Creative|TeamOption|GetLegacyTeamIndex (Variables|UserOptions|GetAssignedTeam)) (Variables|UserOptions|GetShowObjectivePulseToFriendlyPlayers) false (Variables|UserOptions|GetShowObjectivePulseToInstigatorOnly) false 2.0 (Variables|Default|GetCreative_UseRestriction_Component)))

(event OnAddedtoMinigame(FortMinigameProgress)
  (params (Minigame :type Fort%20Minigame%20Object%20Reference))
  ;[1 1 1] Optimize Use Restriction Usage by doing the update based on Team / Class change event intead of always validating it per tick
  (Networking|CanExecuteCosmeticEvents
    (:False
      (Utilities|IsValid Minigame
        (:Is%20Valid
          (Minigame|AssignOnPlayerJoininProgress Minigame (EventDispatchers|CreateEvent :!event Custom|OnPlayerJoinInProgress_Event)))))))

(event Custom|OnPlayerJoinInProgress_Event
  (params (Pawn :type Pawn%20Object%20Reference))
  (Utilities|IsValid Pawn
    (:Is%20Valid
      (CallFunction|AddorUpdateIndicatorControllerComponent (Pawn|GetController Pawn)))))

(fn ActivateFromOtherDevice
  (params (Instigator :type Controller%20Object%20Reference) (Requester :type ERequesterType%20Enum :default "NewEnumerator0"))
  (return false))

(fn AddAllControllers
  (for _array_element (Fortnite|GetAllFortPlayerControllers true)
    (CallFunction|AddorUpdateIndicatorControllerComponent _array_element)))

(fn AddOrUpdateIndicatorControllerComponent
  (params (Controller :type Controller%20Object%20Reference))
  (bind _returnvalue (Components|AddControllerComponent Controller "/CreativeCoreDevices/BP_PlayerControllerComponent_MapIndicator.BP_PlayerControllerComponent_MapIndicator_C"))
  (bind _asbp_player_controller_component_map_indicator (Utilities|Casting|CastToBP_PlayerControllerComponent_MapIndicator _returnvalue))
  (Class|BPPlayerControllerComponentMapIndicator|RegisterMarkerComponent _asbp_player_controller_component_map_indicator (Variables|Default|GetCreativeMapMarkerComponent))
  (CallFunction|UpdateComponentState (Class|CreativeEnabledComponent|IsEnabled (Variables|Default|GetEnabledComponent))))

(fn CanInteract
  (params (InteractingPawn :type Fort%20Pawn%20Object%20Reference) (InteractionBeingAttempted :type EInteractionBeingAttempted%20Enum :default "FirstInteraction") (InteractionType :type TInteractionType%20Enum :default "IT_NoInteraction"))
  (return false))

(fn GetDeviceColor
  (return (Variables|UserOptions|Basic|GetIconColor)))

(fn GetInstigatingController
  (return 0))

(fn GetInitialMarkerVisibility
  (bind _currentminigame (Variables|/Script/FortniteGame.FortMinigameLogicComponent|GetCurrentMinigame (Variables|Default|GetFortMinigameProgress))
    (:then
      (Minigame|IsInProgress _currentminigame
        (:Yes
          _)
        (:No
          (return true))))
    (:else
      _))
  (return (Variables|UserOptions|Basic|GetEnabledOnGameStart)))

(fn GetSpawnedPickups
  (return 0))

(fn GetVisibilityComponents
  (bind _visibilitycomponents (|CallParentFunction))
  (Utilities|Array|AppendArray _visibilitycomponents (Utilities|Array|MakeArray (Variables|Debug|GetStaticMeshComponent) (Variables|Default|GetSM_BDA_Attachment_01) (Variables|Default|GetBaseDeviceScreenGlow) (Variables|Default|GetSM_BDA_Attachment-Right) (Variables|Default|GetSM_BDA_Attachment-Left)))
  (return _visibilitycomponents))

(fn IsEnabled
  (return false))

(fn IsHoldingItemForPawn
  (params (Item :type Item%20Variant%20Handle%20Structure) (Pawn :type Pawn%20Object%20Reference))
  (return false))

(fn IsTargetableByLinearMover
  (return true))

(fn MigrateLegacyUserOptions
  (bind (_knob_affectsclass _knob_affectsteam) (Class|GetClassDefaults "/CreativeCoreDevices/Device_MapIndicator_V2.Device_MapIndicator_V2_C"))
  (bind _knob_iconcolor (Class|GetClassDefaults "/CreativeCoreDevices/Device_MapIndicator_V2.Device_MapIndicator_V2_C"))
  (|CallParentFunction)
  (Utilities|FlowControl|Sequence
    (:then_0
      (Utilities|FlowControl|Sequence
        (:then_0
          (if (!= (Variables|Legacy|GetKnob_AffectsClass) _knob_affectsclass)
            (Variables|UserOptions|SetInvertClass true)
            (Variables|Legacy|SetKnob_AffectsClass _knob_affectsclass)))
        (:then_1
          (if (!= _knob_affectsteam (Variables|Legacy|GetKnob_AffectsTeam))
            (Variables|UserOptions|SetInvertTeam true)
            (Variables|Legacy|SetKnob_AffectsTeam _knob_affectsteam)))
        (:then_2
          (if (Utilities|Enum|NotEqual(Enum) (Variables|Legacy|GetKnob_IconColor) _knob_iconcolor)
            (Variables|UserOptions|Basic|SetIconColor (Utilities|Select :Index (Variables|Legacy|GetKnob_IconColor) :None "(R=0.000000,G=0.537697,B=1.000000,A=1.000000)" :White "(R=0.352941,G=0.352941,B=0.352941,A=1.000000)" :Red "(R=0.427451,G=0.035294,B=0.047059,A=1.000000)" :Orange "(R=0.670588,G=0.207843,B=0.023529,A=1.000000)" :Yellow "(R=0.800000,G=0.572549,B=0.011765,A=1.000000)" :Green "(R=0.062745,G=0.298039,B=0.066667,A=1.000000)" :Teal "(R=0.003922,G=0.235294,B=0.356863,A=1.000000)" :Blue "(R=0.039216,G=0.047059,B=0.305882,A=1.000000)" :Purple "(R=0.498039,G=0.094118,B=0.301961,A=1.000000)"))
            (Variables|Legacy|SetKnob_IconColor _knob_iconcolor)))))
    (:then_1
      (OptionData|RefreshPropertyOverrides (Variables|Default|GetToyOptionsComponent)))))

(fn SetTriggerCollision
  (params (On :type Boolean))
  (return false))

(fn ShouldInvisiblePropsBeAllowed
  (return true))

(fn UpdateComponentState
  (params (bVisible :type Boolean))
  (MiniMapData|SetMiniMapIndicatorIsVisible (Variables|Default|GetCreativeMapMarkerComponent) bVisible))

(fn ConstructionScript
  (|CallParentFunction)
  (Networking|CanExecuteCosmeticEvents
    (:True
      (Transformation|AttachComponentToComponent (Variables|Default|GetSM_BDA_Attachment_01) (Variables|Debug|GetStaticMeshComponent) "Attachment_Bottom" "KeepRelative" "KeepRelative" "KeepRelative" true)
      (Transformation|AttachComponentToComponent (Variables|Default|GetSM_BDA_Attachment-Right) (Variables|Debug|GetStaticMeshComponent) "Attachment_RightMiddle" "KeepRelative" "KeepRelative" "KeepRelative" true)
      (Transformation|AttachComponentToComponent (Variables|Default|GetSM_BDA_Attachment-Left) (Variables|Debug|GetStaticMeshComponent) "Attachment_LeftMiddle" "KeepRelative" "KeepRelative" "KeepRelative" true)
      (Utilities|SetCreativeFakeGlowParameters (Variables|Default|GetBaseDeviceScreenGlow) (Variables|Default|GetGlowColor) (Variables|Default|GetFlickerIntensity) (Variables|Default|GetFlickerSpeed) (Variables|Default|GetGlowOpacity)
        (:ExecOut
          (Rendering|Material|SetCustomPrimitiveDataFloat (Variables|Debug|GetStaticMeshComponent) 27 17.0))))))
