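; Creative Movement Modulator device: a pad that launches actors and/or changes their movement speed.
; Trigger overlap or an Activate event checks use restrictions, applies the impulse, then on authority applies a SetByCaller speed GameplayEffect and broadcasts activation with a reset delay.
; Per-actor effect handles let Deactivate events remove the effect.
; Enabled state and options drive trigger collision, pad color and idle particles; rate-limited multicast boost FX load asynchronously.
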(event Custom|Enable%20Event
  (params (Instigator :type Controller%20Object%20Reference))
  ;[0.8046 0 1 0.5] Gameplay Functions
  (Class|CreativeEnabledComponent|SetEnabled (Variables|Default|GetEnabledComponent) true))

(event Custom|Deactivate%20Event
  (params (Instigator :type Controller%20Object%20Reference))
  (bind _returnvalue (Pawn|GetControlledPawn Instigator))
  ; Verse Functions
  (Utilities|IsValid Instigator
    (:Is%20Valid
      (Utilities|IsValid _returnvalue
        (:Is%20Valid
          (CallFunction|RemoveMovementEffect _returnvalue))))))

(event EventBeginPlay
  ;[0 1 0.3143 0.5] Initialization
  (|CallParentFunction)
  (LOD|SetMaxDrawDistance (Variables|Default|GetP_SpeedBoost_Idle) 5e+03))

(event Custom|Disable%20Event
  (params (Instigator :type Controller%20Object%20Reference))
  ;[0.8046 0 1 0.5] Gameplay Functions
  (Class|CreativeEnabledComponent|SetEnabled (Variables|Default|GetEnabledComponent) false))

(event Custom|Deactivate%20For%20All%20Event
  ; Verse Functions
  (CallFunction|RemoveMovementEffectAll))

(event Custom|Activate%20Event
  (params (Instigator :type Controller%20Object%20Reference))
  ;[0.8046 0 1 0.5] Gameplay Functions
  (if (Class|CreativeEnabledComponent|IsEnabled (Variables|Default|GetEnabledComponent))
    (Utilities|IsValid Instigator
      (:Is%20Valid
        (CallFunction|AttemptActivateForActor (Pawn|GetControlledPawn Instigator) true)))))

(event EditorPreview|EventSetupForEditorPreview
  (bind _visibleingamecomponent (Variables|Default|GetVisibleInGameComponent))
  ;[0 1 0.3143 0.5] Initialization
  (|CallParentFunction)
  (Initialization|Initialize (Variables|Default|GetCreative_UseRestriction_Component) (Variables|Default|GetFortMinigameProgress) (Variables|Default|GetEnabledComponent) 0 (Variables|UserOptions|GetActivatingTeam) (Variables|UserOptions|GetInvertTeamSelection) (Variables|UserOptions|GetActivatingClass) (Variables|UserOptions|GetInvertClassSelection) false)
  (Visibility|UpdateVisibility _visibleingamecomponent)
  (CallFunction|UpdateBaseCollision (Visibility|ShouldbeVisible _visibleingamecomponent))
  (CallFunction|UpdateDeviceGameplay))

(event EventInitializeDevice
  (bind _visibleingamecomponent (Variables|Default|GetVisibleInGameComponent))
  ;[0 1 0.3143 0.5] Initialization
  (|CallParentFunction)
  (Initialization|Initialize (Variables|Default|GetCreative_UseRestriction_Component) (Variables|Default|GetFortMinigameProgress) (Variables|Default|GetEnabledComponent) 0 (Variables|UserOptions|GetActivatingTeam) (Variables|UserOptions|GetInvertTeamSelection) (Variables|UserOptions|GetActivatingClass) (Variables|UserOptions|GetInvertClassSelection) false)
  (Visibility|UpdateVisibility _visibleingamecomponent)
  (CallFunction|UpdateBaseCollision (Visibility|ShouldbeVisible _visibleingamecomponent))
  (CallFunction|UpdateDeviceGameplay))

(event Custom|Update%20Device%20Gameplay
  ;[0 0.5217 1 0.5] Update Device Logic
  (CallFunction|SetSpeedState)
  (CallFunction|SetDeviceVisuals)
  (Collision|SetCollisionEnabled (Variables|Trap_Floor_Player_Jump_Pad|GetTrigger) (select (Class|CreativeEnabledComponent|IsEnabled (Variables|Default|GetEnabledComponent)) "QueryOnly" "NoCollision"))
  (Collision|SetCollisionResponsetoChannel (Variables|Trap_Floor_Player_Jump_Pad|GetTrigger) "ECC_GameTraceChannel4" (select (Variables|UserOptions|GetApply%20To%20Non-Player%20Character) "ECR_Overlap" "ECR_Ignore")))

(event Custom|Attempt%20Activate%20for%20Actor
  (params (Actor :type Actor%20Object%20Reference) (Remotely%20Triggered :type Boolean))
  ;[0 0.4302 1 0.5] Device Triggered Logic
  (bind (_valid_controller _fortplayerpawn _controller) (CreativeUserRestrictions|ActorHasValidParticipantController (Variables|Default|GetCreative_UseRestriction_Component) Actor true true true))
  (if _valid_controller
    _
    (elif (CreativeUserRestrictions|CanParticipantActorInteract (Variables|Default|GetCreative_UseRestriction_Component) Actor true true false)
      _))
  (bind _any_impulse_applied (CallFunction|ImpulseActor Actor Remotely%20Triggered))
  (Networking|SwitchHasAuthority
    (:Authority
      (bind _any_effect_applied (CallFunction|ApplyMovementEffects Actor))
      (if (or _any_impulse_applied _any_effect_applied)
        (CallFunction|DeviceTriggered _controller)))))

(event Custom|Device%20Triggered
  (params (TriggerInstigator :type Controller%20Object%20Reference))
  ;[0 0.4302 1 0.5] Device Triggered Logic
  (Reset|ResetDelayStart (Variables|Default|GetEnabledComponent) (Variables|UserOptions|GetResetDelay))
  (GameplayEvent|BroadcastEvent (Variables|UserOptions|Events|GetOnActivation) 0 (Utilities|Struct|MakeGameplayEventActivated TriggerInstigator)))

(event OnComponentBeginOverlap(Trigger)
  (params (OverlappedComponent :type Primitive%20Component%20Object%20Reference) (OtherActor :type Actor%20Object%20Reference) (OtherComp :type Primitive%20Component%20Object%20Reference) (OtherBodyIndex :type Integer) (bFromSweep :type Boolean) (SweepResult :type Hit%20Result%20Structure%20%28by%20ref%29))
  ;[0 1 0.1063] Actor Overlapped Trigger
  (CallFunction|AttemptActivateForActor OtherActor))

(event OnEnabledStateChanged(EnabledComponent)
  (params (Enabled :type Boolean))
  ;[0 1 0.7065 0.5] Component Callbacks
  ;;[0 1 0.9331 0] Enabled State Updating
  (CallFunction|UpdateDeviceGameplay))

(event OnVisibilityChanged(VisibleInGameComponent)
  (params (Visible :type Boolean))
  ;[0 1 0.7065 0.5] Component Callbacks
  ;;[0 1 0.9331 0] Visibility State Updating
  (CallFunction|UpdateBaseMeshinEditMode Visible)
  (CallFunction|UpdateBaseCollision Visible)
  (Networking|CanExecuteCosmeticEvents
    (:True
      (Rendering|SetVisibility (Variables|Default|GetP_SpeedBoost_Idle) (and (Math|Boolean|ORBoolean Visible (Utilities|Enum|Equal(Enum) (Variables|UserOptions|Basic|GetVisibleDuringGame) "NewEnumerator1") (not (CallFunction|IsinPlayMode))) (Utilities|Enum|NotEqual(Enum) (Variables|UserOptions|Basic|GetVisibleDuringGame) "NewEnumerator3")) false))))

(event Curie|EventOnCurieContainerReparented
  (params (CurieContainerHandle :type Curie%20Container%20Handle%20Structure)))

(event Custom|SetDeviceVisuals
  ;[1 0.7825 0 0.5] Client - Update Trap Visuals with the current state
  (Networking|CanExecuteCosmeticEvents
    (:True
      (if (Visibility|ShouldbeVisible (Variables|Default|GetVisibleInGameComponent))
        (CallFunction|UpdatePadColor)
        (switch Utilities|FlowControl|Switch|SwitchonECreativeSpeedModType (Variables|Default|GetSpeedState)
          (:NewEnumerator0
            (if (not (Math|Boolean|NANDBoolean (CallFunction|IsinPlayMode) (or (Utilities|Enum|Equal(Enum) (Variables|UserOptions|Basic|GetVisibleDuringGame) "NewEnumerator2") (Utilities|Enum|Equal(Enum) (Variables|UserOptions|Basic|GetVisibleDuringGame) "NewEnumerator3"))))
              (Components|Activation|Deactivate (Variables|Default|GetP_SpeedBoost_Idle))))
          (:NewEnumerator1
            _)
          (:NewEnumerator2
            _)
          (:NewEnumerator3
            _))
        (Development|SetHiddeninGame (Variables|Default|GetP_SpeedBoost_Idle) (Editor|InEditor) false) ; If we're editor, we can hide in game here.
        (Components|Activation|Activate (Variables|Default|GetP_SpeedBoost_Idle) false)
        (CallFunction|UpdateParticleEffects)
        (else
          _)))))

(event Custom|Play%20Speed%20Boost%20FX
  ;[1 0.7374 0 0.2] Async load and play Speed Boost FX (Async as we may not need the effects)
  (Utilities|FlowControl|DoOnce false
    (:Completed
      (CallFunction|MulticastPlaySpeedBoostFX)
      (Utilities|FlowControl|Delay (Variables|Default|GetSpeed%20Boost%20Effect%20Limiter)))))

(event Custom|Multicast_PlaySpeedBoostFX
  ;[1 0.7374 0 0.2] Async load and play Speed Boost FX (Async as we may not need the effects)
  (Networking|CanExecuteCosmeticEvents
    (:True
      (Utilities|FlowControl|Sequence
        (:then_0
          (bind _speed_boost_cue (Variables|SpeedBoost|GetSpeed%20Boost%20Cue)
            (:then
              (bind _object (Utilities|AsyncLoadAsset _speed_boost_cue)
                (:Completed
                  (bind _assound_cue (Utilities|Casting|CastToSoundCue _object))
                  (Audio|PlaySoundatLocation _assound_cue (Transformation|GetWorldLocation (Variables|Debug|GetStaticMeshComponent))))))
            (:else)))
        (:then_1
          (bind _speed_boost_particle (Variables|SpeedBoost|GetSpeed%20Boost%20Particle)
            (:then
              (bind _object_2 (Utilities|AsyncLoadAsset _speed_boost_particle)
                (:Completed
                  (bind _asparticle_system (Utilities|Casting|CastToParticleSystem _object_2))
                  (Effects|SpawnEmitterAttached _asparticle_system (Variables|Debug|GetStaticMeshComponent) "None" "0.000000,0.000000,30.000000" "0.000000,90.000000,0.000000" "1.000000,1.000000,1.000000" "KeepRelativeOffset" true "AutoRelease"))))
            (:else)))))))

(fn ApplyMovementEffects
  (params (Actor :type Actor%20Object%20Reference))
  (Networking|SwitchHasAuthority
    (:Authority
      (if (Variables|UserOptions|Basic|GetAffect%20Movement%20Speed)
        (Utilities|IsValid Actor
          (:Is%20Valid
            (LocalVariables|SetASC (Ability|GetAbilitySystemComponent Actor))
            (bind _asc (LocalVariables|GetASC)
              (:then
                (GameplayEffects|RemoveActiveEffectswithTags _asc "(GameplayTags=((TagName=\"Gameplay.Mod.RunSpeed\"),(TagName=\"Gameplay.Mod.SprintSpeed\"),(TagName=\"Gameplay.Mod.SwimSpeed\")))")
                (LocalVariables|SetOutgoingSpec_MovementMod (GameplayEffects|MakeOutgoingSpec _asc (select (and (not (Variables|/Script/Engine.Actor|GetbHidden (Variables|GetActor))) (Variables|UserOptions|GetShow%20Movement%20Effects%20on%20Players)) (Variables|Default|GetMovementGEClass) (Variables|Default|GetMovementGEClass_NoCue)) (Utilities|Select :Index (Variables|Default|GetSpeedState) :NewEnumerator0 0.0 :NewEnumerator1 2.0 :NewEnumerator2 3.0 :NewEnumerator3 4.0) (GameplayEffects|MakeEffectContext _asc)))
                (if (not (Variables|UserOptions|Basic|GetInfiniteDuration))
                  (Ability|GameplayEffect|SetDuration (LocalVariables|GetOutgoingSpec_MovementMod) (Variables|UserOptions|Basic|GetEffectDuration)))
                (bind _returnvalue (Ability|GameplayEffect|AssignTagSetbyCallerMagnitude (LocalVariables|GetOutgoingSpec_MovementMod) "(TagName=\"SetByCaller.MoveSpeed\")" (Variables|UserOptions|Basic|GetSpeed)))
                (bind _returnvalue_2 (GameplayEffects|ApplyGameplayEffectSpecToTarget _asc _returnvalue _asc))
                (Utilities|Map|Add (Variables|Default|GetActiveMovementEffectHandles) Actor _returnvalue_2)
                (return true))
              (:else))))))))

(fn DeactivateForAllVerse
  (CallFunction|DeactivateforAllEvent))

(fn DeactivateVerse
  (params (Instigator :type Controller%20Object%20Reference))
  (CallFunction|DeactivateEvent Instigator))

(fn GetEnabledStateasIndex
  (return (DeviceConversion|EnabledPhaseEnumtoLegacyInt (Variables|UserOptions|GetEnabledDuringPhase))))

(fn GetCollisionStaticMesh
  (return "0, 0, 0" (Variables|/Script/Engine.StaticMeshComponent|GetStaticMesh (Variables|Default|GetBaseMesh))))

(fn GetCreativeActorBounds
  (bind (_origin _boxextent) (Collision|GetComponentBounds (Variables|Default|GetBaseMesh)))
  (return true _boxextent))

(fn GetCreativeActorOrigin
  (return true (Variables|/Script/Engine.SceneComponent|GetRelativeLocation (Variables|Default|GetBaseCollision))))

(fn GetCreativeActorRelativeTransform
  (return false 0))

(fn GetGridAxisGuideAttachmentPoint
  (return false false "0, 0, 0"))

(fn GetIconPlacement
  (params (SelfActor :type Actor%20Object%20Reference) (ViewingActor :type Actor%20Object%20Reference))
  (return (Physics|GetCenterOfMass (Variables|Trap_Floor_Player_Jump_Pad|GetTrigger) "None") (Variables|/Script/Engine.BoxComponent|GetBoxExtent (Variables|Trap_Floor_Player_Jump_Pad|GetTrigger))))

(fn GetMeshComponents
  (return (Utilities|Array|MakeArray (Variables|Default|GetBaseMesh))))

(fn GetOverrideMeshMaterial
  (params (MeshComp :type Static%20Mesh%20Component%20Object%20Reference) (MatIdx :type Integer))
  ;[0 1 0.9331 0.25] Normal if harsh behavior override. If the phone tool inquires about materials, forceably reassign our material to the base as it's probably screwed with it.
  (if (== (Variables|Default|GetBaseCollision) MeshComp)
    (Rendering|Material|SetMaterial (Variables|Default|GetBaseCollision) 0 (Variables|Colors|GetMeshMID))
    (return (Variables|Colors|GetMeshMID))
    (else
      (return (|CallParentFunction MeshComp MatIdx)))))

(fn IgnoreWorldGridScaleOverride
  (return false))

(fn ImpulseActor
  (params (Actor :type Actor%20Object%20Reference) (Remotely%20Triggered :type Boolean))
  (bind _forward_impulse (Variables|UserOptions|Basic|GetForward%20Impulse))
  (bind _impulse_forward (LocalVariables|GetImpulse%20Forward))
  (bind _returnvalue (* _impulse_forward _forward_impulse))
  (bind _apply_upward_impulse_to_non_characters (Variables|UserOptions|GetApply%20Upward%20Impulse%20to%20Non-Characters))
  (bind _upward_impulse (Variables|UserOptions|GetUpward%20Impulse))
  (bind _returnvalue_2 (Math|Float|SelectFloat _upward_impulse 0.0 _apply_upward_impulse_to_non_characters))
  (bind _returnvalue_3 (Utilities|Operators|Add _returnvalue (Math|Vector|MakeVector 0.0 0.0 _returnvalue_2)))
  (LocalVariables|SetActor%20to%20Check Actor)
  (if (Variables|UserOptions|Basic|GetApply%20Impulse)
    (LocalVariables|SetImpulse%20Forward (Transformation|GetActorForwardVector (select (and Remotely%20Triggered (Variables|UserOptions|GetApply%20Remote%20Impulse%20Relative%20to%20Player%20Direction)) (LocalVariables|GetActor%20to%20Check) self)))
    (bind _ascharacter (Utilities|Casting|CastToCharacter (LocalVariables|GetActor%20to%20Check))
      (:then
        (if (and (Variables|GetRemotely%20Triggered) (Variables|UserOptions|GetApply%20Remote%20Impulse%20Relative%20to%20Player%20Direction))
          (Utilities|IsValid (Fortnite|GetFortPlayerControllerfromActor _ascharacter)
            (:Is%20Valid
              (bind (_returnvalue_x _returnvalue_y) (Math|Vector|BreakVector (Math|Vector|GetForwardVector (Camera|GetCameraRotation (Variables|/Script/Engine.PlayerController|GetPlayerCameraManager (Fortnite|GetFortPlayerControllerfromActor _ascharacter))))))
              (LocalVariables|SetImpulse%20Forward (Math|Vector|Normalize (Math|Vector|MakeVector _returnvalue_x _returnvalue_y))))
            (:Is%20Not%20Valid
              (bind (_returnvalue_x_2 _returnvalue_y_2) (Math|Vector|BreakVector (Transformation|GetActorForwardVector _ascharacter)))
              (LocalVariables|SetImpulse%20Forward (Math|Vector|Normalize (Math|Vector|MakeVector _returnvalue_x_2 _returnvalue_y_2))))))
        (Character|LaunchCharacter _ascharacter (Utilities|Operators|Add (Math|Vector|MakeVector 0.0 0.0 (* (Variables|/Script/Engine.CharacterMovementComponent|GetGravityScale (Variables|/Script/Engine.Character|GetCharacterMovement _ascharacter)) (Variables|UserOptions|GetUpward%20Impulse))) (* (LocalVariables|GetImpulse%20Forward) (Variables|UserOptions|Basic|GetForward%20Impulse))) true true))
      (:CastFailed
        (Utilities|Casting|CastToFortVehicleInterface (LocalVariables|GetActor%20to%20Check)
          (:then
            (Audio|BroadcastAbilityImpactatLocation 0 (LocalVariables|GetActor%20to%20Check) _returnvalue_3 777.0) ; Cannot apply physics impulse to manned\nvehicles for some reason.
          )
          (:CastFailed
            (FortPhysicsSimulationLibrary|AddPhysicsImpulse (LocalVariables|GetActor%20to%20Check) _returnvalue_3 true "None" true false self "(MaxSimSizeAllowed=Large,bRejectEmbeddedInLandscape=False)")))))
    (CallFunction|PlaySpeedBoostFX) ; If we impulsed anything, play Speed Boost FX
    (return true)))

(fn MigrateLegacyUserOptions
  (bind _impulsepresetindex (Class|GetClassDefaults "/CreativeCoreDevices/Device_MovementModulator_V2.Device_MovementModulator_V2_C"))
  (bind _duration (Class|GetClassDefaults "/CreativeCoreDevices/Device_MovementModulator_V2.Device_MovementModulator_V2_C"))
  (bind _impulsepresetindex_2 (Variables|LegacyOptions|GetImpulsePresetIndex))
  (bind _duration_2 (Variables|LegacyOptions|GetDuration))
  (Utilities|FlowControl|Sequence
    (:then_0
      (Utilities|FlowControl|Sequence
        (:then_0
          ;[1 0.5668 0 0.5] Migrating Forward Impulse
          (if (!= _impulsepresetindex_2 _impulsepresetindex)
            (Variables|UserOptions|Basic|SetForward%20Impulse (Math|Conversions|ToFloat(Integer) (* _impulsepresetindex_2 1000)))
            (Variables|LegacyOptions|SetImpulsePresetIndex _impulsepresetindex)
            (LocalVariables|SetOptionMigrated true)))
        (:then_1
          ;[0 1 0.452 0.5] Migrating Duration
          (if (!= _duration_2 _duration)
            (Variables|UserOptions|Basic|SetInfiniteDuration (== _duration_2 1.8e+04))
            (Variables|LegacyOptions|SetDuration _duration)
            (LocalVariables|SetOptionMigrated true)))))
    (:then_1
      (if (LocalVariables|GetOptionMigrated)
        (OptionData|RefreshPropertyOverrides (Variables|Default|GetToyOptionsComponent))))))

(fn RemoveMovementEffect
  (params (Actor :type Actor%20Object%20Reference))
  (bind _activemovementeffecthandles (Variables|Default|GetActiveMovementEffectHandles))
  (bind (_value _returnvalue) (Utilities|Map|Find _activemovementeffecthandles Actor))
  (Networking|SwitchHasAuthority
    (:Authority
      (if (Variables|UserOptions|Basic|GetAffect%20Movement%20Speed)
        (Utilities|IsValid Actor
          (:Is%20Valid
            (LocalVariables|SetASC (Ability|GetAbilitySystemComponent Actor))
            (bind _asc (LocalVariables|GetASC)
              (:then
                (if _returnvalue
                  (GameplayEffects|RemoveActiveGameplayEffect _asc _value -1)
                  (Utilities|Map|Remove (Variables|Default|GetActiveMovementEffectHandles) Actor)))
              (:else))))))))

(fn RemoveMovementEffectAll
  (bind _activemovementeffecthandles (Variables|Default|GetActiveMovementEffectHandles))
  (Networking|SwitchHasAuthority
    (:Authority
      (bind (_keypin _valuepin) (Utilities|Map|ForEachLoop(Map) _activemovementeffecthandles)
        (:then
          (LocalVariables|SetASC (Ability|GetAbilitySystemComponent _keypin))
          (bind _asc (LocalVariables|GetASC)
            (:then
              (GameplayEffects|RemoveActiveGameplayEffect _asc _valuepin -1))
            (:else)))
        (:CompletedPin
          (Utilities|Map|Clear _activemovementeffecthandles))))))

(fn SetSpeedState
  (Variables|Default|SetSpeedState (select (Class|CreativeEnabledComponent|IsEnabled (Variables|Default|GetEnabledComponent)) (select (== (Variables|UserOptions|Basic|GetSpeed) 1.0) "NewEnumerator2" (select (> (Variables|UserOptions|Basic|GetSpeed) 1.0) "NewEnumerator3" "NewEnumerator1")) "NewEnumerator0")))

(fn UpdateBaseMeshInEditMode
  (params (IsVisible :type Boolean))
  (bind _basemesh (Variables|Default|GetBaseMesh))
  (bind _basemesh_2 (Variables|Default|GetBaseMesh))
  ;[1 1 1] If a minigame is not in progress we're In VK-Edit / Play and should override visibility / change mesh depending on visibile state
  (bind _current_minigame (Utilities|GetCurrentMinigame)
    (:Is%20Valid
      (if (Minigame|IsinProgress _current_minigame)
        _
        (else
          _)))
    (:Is%20Not%20Valid
      _))
  (Rendering|SetVisibility _basemesh_2 true false)
  (Development|SetHiddeninGame _basemesh_2 false false)
  (if IsVisible
    (Rendering|Material|SetMaterial _basemesh 0 "/Game/Weapons/FORT_TRAPS/Trap_SpeedBoosts/Materials/M_Speedboost_Straight.M_Speedboost_Straight")
    (else
      (Rendering|Material|SetMaterial _basemesh 0 "/Game/Creative/Devices/Generic/Materials/M_CP_Device_Hologram_Vehicles.M_CP_Device_Hologram_Vehicles"))))

(fn UpdatePadColor
  (bind _meshmid (Variables|Colors|GetMeshMID))
  (Material|ReattachOrCreateMID (Variables|Default|GetBaseMesh) _meshmid)
  (Rendering|Material|SetVectorParameterValue _meshmid "EmissiveMultiplier" (Math|Color|SelectColor (* (Variables|UserOptions|GetColor) (Variables|Colors|GetBase%20Color%20Emissive%20Multiplier)) (Math|Color|SelectColor (Utilities|Map|Find (Variables|Colors|GetSpeedEmissives) (Variables|Default|GetSpeedState)) (Math|Color|SelectColor (Variables|Colors|GetImpulse%20Color) (Variables|Colors|GetDefault%20Color) (Variables|UserOptions|Basic|GetApply%20Impulse)) (Variables|UserOptions|Basic|GetAffect%20Movement%20Speed)) (Variables|UserOptions|GetUseCustomColor)) :!class /Script/Engine.MaterialInstanceDynamic))

(fn UpdateParticleEffects
  (bind _color (Variables|UserOptions|GetColor))
  (bind _returnvalue (Rendering|Material|CreateDynamicMaterialInstance (Variables|Default|GetP_SpeedBoost_Idle) 0 (select (Variables|UserOptions|GetUseCustomColor) (Variables|Colors|GetDefault%20Particle%20Material) (select (Variables|UserOptions|Basic|GetAffect%20Movement%20Speed) (Utilities|Map|Find (Variables|Colors|GetSpeedMaterials) (Variables|Default|GetSpeedState)) (select (Variables|UserOptions|Basic|GetApply%20Impulse) (Variables|SpeedBoost|GetImpulse%20Particle%20Material) (Variables|Colors|GetDefault%20Particle%20Material)))) "None" :!class /Script/Engine.PrimitiveComponent)) ; Pick Material to use. Prefer in this order:\nCustom Color\nMovement Speed\nImpulse\nNone
  (Variables|/Script/CascadeRuntime.ParticleSystemComponent|SetCustomTimeDilation (Math|Float|Max(Float) (Math|Float|MapRangeClamped (Math|Float|SelectFloat (Variables|UserOptions|Basic|GetSpeed) 0.0 (Variables|UserOptions|Basic|GetAffect%20Movement%20Speed)) 0.5 3.0 0.33 2.0) (Math|Float|MapRangeClamped (Math|Float|SelectFloat (Variables|UserOptions|Basic|GetForward%20Impulse) 0.0 (Variables|UserOptions|Basic|GetApply%20Impulse)) 833.0 5e+03 0.33 2.0)) (Variables|Default|GetP_SpeedBoost_Idle))
  (if (Variables|UserOptions|GetUseCustomColor)
    (Rendering|Material|SetVectorParameterValue _returnvalue "Glow Color Inner" (* (Variables|Colors|GetParticle%20Color%20Inner%20Multiplier) _color) :!class /Script/Engine.MaterialInstanceDynamic)
    (Rendering|Material|SetVectorParameterValue _returnvalue "Glow Color Outer" (* (Variables|Colors|GetParticle%20Color%20Outer%20Multiplier) _color) :!class /Script/Engine.MaterialInstanceDynamic)))

(fn UpdateVisibilityinGame
  (Visibility|SetVisibilityComponents (Variables|Default|GetVisibleInGameComponent) (Utilities|Array|MakeArray (Variables|Default|GetBaseMesh)) false)
  (Visibility|SetVisiblityDuringPhase (Variables|Default|GetVisibleInGameComponent) (Utilities|Select :Index (Variables|UserOptions|Basic|GetVisibleDuringGame) :NewEnumerator0 "NewEnumerator0" :NewEnumerator1 "NewEnumerator0" :NewEnumerator2 "NewEnumerator3" :NewEnumerator3 "NewEnumerator3")))

(fn UpdateBaseCollision
  (params (Base%20Visibility :type Boolean))
  (Collision|SetCollisionEnabled (Variables|Default|GetBaseMesh) (select (CallFunction|IsinPlayMode) "NoCollision" "QueryOnly")) ; Base Mesh collision is used purely for \ninteraction during edit mode in order to \nopen the customize menu.
  (Collision|SetCollisionEnabled (Variables|Default|GetBaseCollision) (select (and Base%20Visibility (Variables|UserOptions|GetPad%20has%20Collision)) "QueryAndPhysics" "NoCollision")) ; Only give the base collision if we need it.\nThis is full collision if it's visible.
)

(fn ConstructionScript)
