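; Creative Objective device: a damageable, team-owned target that reports health to the minigame as tracked progress and broadcasts damage and destruction events.
; User options drive the async-loaded mesh, beacon particle, team-aware health badge and objective trail.
; Runtime invulnerability and forced visibility/collision change mid-game without altering saved settings.
; Destruction can play explosion effects and run a throttled async overlap that destroys nearby damageable actors, chaining into other objectives.
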(event EditorPreview|EventSetupForEditorPreview
  ;[0 0.4635 0.02851 0.1] World Ready
  (Visibility|SetVisibilityComponents (Variables|Default|GetVisibleInGameComponent) (Utilities|Array|MakeArray (Variables|Debug|GetStaticMeshComponent)) true)
  (Audio|Parameter|SetIntegerParameter (Variables|Default|GetDeath%20Sound) "ObjectIndex" (Math|Conversions|ToInteger(Byte) (Variables|UserOptions|Basic|GetMesh)) :!class /Script/AudioExtensions.AudioParameterControllerInterface)
  (CallFunction|SetParticle))

(event EventOnWorldReady
  ;[0 0.4635 0.02851 0.1] World Ready
  (|CallParentFunction)
  (Utilities|FlowControl|Sequence
    (:then_0
      (Visibility|SetVisibilityComponents (Variables|Default|GetVisibleInGameComponent) (Utilities|Array|MakeArray (Variables|Debug|GetStaticMeshComponent)) true)
      (Audio|Parameter|SetIntegerParameter (Variables|Default|GetDeath%20Sound) "ObjectIndex" (Math|Conversions|ToInteger(Byte) (Variables|UserOptions|Basic|GetMesh)) :!class /Script/AudioExtensions.AudioParameterControllerInterface)
      (CallFunction|SetParticle))
    (:then_1
      (Networking|CanExecuteCosmeticEvents
        (:True
          (Utilities|FlowControl|Sequence
            (:then_0
              ;;[1 0.9946 0 0.584] client only-create the hp bar ui
              (bind _class (Utilities|AsyncLoadClassAsset (Variables|UI|GetHealthBarSoftClassRef))
                (:Completed
                  (bind _asuser_widget (Utilities|Casting|CastToUserWidgetClass _class))
                  (bind _returnvalue (UserInterface|CreateWidget _asuser_widget (Game|GetPlayerController 0)))
                  (UserInterface|SetWidget (Variables|Default|GetHealthBarComp) _returnvalue)
                  (Init|InitfromObject(Message) _returnvalue self))))
            (:then_1
              (for _array_element (Fortnite|GetLocalFortPlayerControllers false true)
                (bind _asfort_player_controller_spectating (Utilities|Casting|CastToFortPlayerControllerSpectating _array_element))
                (EventDispatchers|BindEventtoOnFollowedPlayerChangedTeam _asfort_player_controller_spectating (EventDispatchers|CreateEvent :!event UpdateFromTeamChanged))
                (EventDispatchers|BindEventtoOnCameraTypeChanged _asfort_player_controller_spectating (EventDispatchers|CreateEvent :!event Custom|OnSpectatorCameraTypeChanged)))
              (CallFunction|UpdatefromTeamChanged))))))
    (:then_2
      (CallFunction|SetShouldTickHealthBarComponent))))

(event Custom|SetVisibilityForcedDuringGame
  (params (NewForcedValue :type Enum_ForcedValueDuringGame%20Enum :default "NewEnumerator0"))
  ;[0.599 0 1 0.2] ForcedValuesDuringGame
  (Networking|SwitchHasAuthority
    (:Authority
      (if (Utilities|Enum|Equal(Enum) (Variables|UserOptions|GetCollision%20During%20Games) "NewEnumerator3")
        (CallFunction|SetCollisionForcedDuringGame NewForcedValue)
        (else
          _))
      (Visibility|SetForcedVisibilityDuringGame (Variables|Default|GetVisibleInGameComponent) NewForcedValue))))

(event OnAnyPropertyChanged(ToyOptionsComponent)
  ;[0 1 0.0615 0.1] On Initialize
  (Utilities|FlowControl|Sequence
    (:then_0
      (CallFunction|PickMesh)
      (Gameplay|ForceBuildingHealth (Math|Float|Max(Float) (Variables|UserOptions|Health|GetHealth) 1.0))
      (Default|CallUpdateRefresh false)
      (CallFunction|UpdateHealthbarHeight)
      (Variables|/Script/FortniteGame.FortWidgetComponent|SetbClampProjectedLocalPositionToViewport (Variables|UserOptions|UI|GetClampToScreen) (Variables|Default|GetHealthBarComp))
      (CallFunction|SetShouldTickHealthBarComponent)
      (if (Variables|UserOptions|ScoringHUD|GetDisplayScoreUpdateOnHUD)
        (Class|CreativeUIBroadcastComponent|Initialize (Variables|Default|GetCreative_HUDBroadcast_Component) (Variables|Default|GetFortMinigameProgress) true)))
    (:then_1
      (CallFunction|UpdateTeam))
    (:then_2
      (CallFunction|UpdateComponentVisibility))
    (:then_3
      (Default|UpdateStaticMeshCollision))))

(event OnPropertiesLoaded(ToyOptionsComponent)
  ;[0 1 0.0615 0.1] On Initialize
  (Utilities|FlowControl|Sequence
    (:then_0
      (CallFunction|PickMesh)
      (Gameplay|ForceBuildingHealth (Math|Float|Max(Float) (Variables|UserOptions|Health|GetHealth) 1.0))
      (Default|CallUpdateRefresh false)
      (CallFunction|UpdateHealthbarHeight)
      (Variables|/Script/FortniteGame.FortWidgetComponent|SetbClampProjectedLocalPositionToViewport (Variables|UserOptions|UI|GetClampToScreen) (Variables|Default|GetHealthBarComp))
      (CallFunction|SetShouldTickHealthBarComponent)
      (if (Variables|UserOptions|ScoringHUD|GetDisplayScoreUpdateOnHUD)
        (Class|CreativeUIBroadcastComponent|Initialize (Variables|Default|GetCreative_HUDBroadcast_Component) (Variables|Default|GetFortMinigameProgress) true)))
    (:then_1
      (CallFunction|UpdateTeam))
    (:then_2
      (CallFunction|UpdateComponentVisibility))
    (:then_3
      (Default|UpdateStaticMeshCollision))))

(event Custom|Initialize%20Device
  ;[0 1 0.0615 0.1] On Initialize
  (Utilities|FlowControl|Sequence
    (:then_0
      (CallFunction|PickMesh)
      (Gameplay|ForceBuildingHealth (Math|Float|Max(Float) (Variables|UserOptions|Health|GetHealth) 1.0))
      (Default|CallUpdateRefresh false)
      (CallFunction|UpdateHealthbarHeight)
      (Variables|/Script/FortniteGame.FortWidgetComponent|SetbClampProjectedLocalPositionToViewport (Variables|UserOptions|UI|GetClampToScreen) (Variables|Default|GetHealthBarComp))
      (CallFunction|SetShouldTickHealthBarComponent)
      (if (Variables|UserOptions|ScoringHUD|GetDisplayScoreUpdateOnHUD)
        (Class|CreativeUIBroadcastComponent|Initialize (Variables|Default|GetCreative_HUDBroadcast_Component) (Variables|Default|GetFortMinigameProgress) true)))
    (:then_1
      (CallFunction|UpdateTeam))
    (:then_2
      (CallFunction|UpdateComponentVisibility))
    (:then_3
      (Default|UpdateStaticMeshCollision))))

(event Custom|SetCollisionForcedDuringGame
  (params (NewForcedValue :type Enum_ForcedValueDuringGame%20Enum :default "NewEnumerator0"))
  ;[0.599 0 1 0.2] ForcedValuesDuringGame
  (Networking|SwitchHasAuthority
    (:Authority
      (if (Utilities|Enum|NotEqual(Enum) NewForcedValue (Variables|Replicated|GetCollisionForcedDuringGame))
        (Variables|Replicated|SetCollisionForcedDuringGame NewForcedValue)))))

(event Custom|Update%20Healthbar%20Height
  (bind (_min _max) (Utilities|Struct|BreakBox (StaticMesh|GetBoundingBox (Variables|/Script/Engine.StaticMeshComponent|GetStaticMesh (Variables|Debug|GetStaticMeshComponent)))))
  (bind _staticmeshcomponent (Variables|Debug|GetStaticMeshComponent))
  (bind _staticmesh (Variables|/Script/Engine.StaticMeshComponent|GetStaticMesh _staticmeshcomponent))
  ;[0 0 0] Healthbar, update content and visibility
  (Utilities|IsValid _staticmeshcomponent
    (:Is%20Valid
      (Utilities|IsValid _staticmesh
        (:Is%20Valid
          (Transformation|SetRelativeLocation (Variables|Default|GetHealthBarComp) (Math|Vector|MakeVector 0.0 0.0 (Math|Float|SelectFloat (* (Variables|UserOptions|UI|GetBadgeHeight) 1e+02) (.z _max) (Variables|UserOptions|UI|GetBadgeHeight_Override))) false false))))))

(event MinigameStateChangedDelegate(FortMinigameLogic)
  (params (Minigame :type Fort%20Minigame%20Object%20Reference) (NewMinigameState :type EFortMinigameState%20Enum :default "PreGame"))
  ;[0.4991 1 0.7068 0.204] State Changes
  (if (Utilities|Enum|Equal(Enum) NewMinigameState "PostGameReset")
    (Variables|Replicated|SetCollisionForcedDuringGame "NewEnumerator0")
    (Default|UpdateStaticMeshCollision)))

(event Custom|StopHealthEffects
  ;[1 0.5502 0] Multicast Effects
  (Components|Activation|SetActive (Variables|Default|GetDamageParticleSystem) false true))

(event OnVisibilityChanged(VisibleInGameComponent)
  (params (Visible :type Boolean))
  ;[0.4991 1 0.7068 0.204] State Changes
  (Default|UpdateStaticMeshCollision))

(event Custom|Play50%25Effect
  (bind _damageparticlesystem (Variables|Default|GetDamageParticleSystem))
  ;[1 0.5502 0] Multicast Effects
  (Components|Activation|SetActive _damageparticlesystem true false)
  (Effects|Components|ParticleSystem|SetEmitterEnable _damageparticlesystem "Sparks" false))

(event Custom|Turn%20on%20Visibility%20Event
  (params (Instigator :type Controller%20Object%20Reference))
  ;[0.2393 0 1 0.25] Gameplay Functions
  (CallFunction|SetVisibilityForcedDuringGame "NewEnumerator1"))

(event Custom|OnSpectatorCameraTypeChanged
  (params (SpectatingPC :type Fort%20Player%20Controller%20Spectating%20Object%20Reference) (CameraType :type ESpectatorCameraType%20Enum :default "ThirdPerson"))
  ;[0.4991 1 0.7068 0.204] State Changes
  (CallFunction|UpdatefromTeamChanged))

(event Custom|Turn%20off%20Visibility%20Event
  (params (Instigator :type Controller%20Object%20Reference))
  ;[0.2393 0 1 0.25] Gameplay Functions
  (CallFunction|SetVisibilityForcedDuringGame "NewEnumerator2"))

(event Custom|Play25%25Effect
  (bind _damageparticlesystem (Variables|Default|GetDamageParticleSystem))
  ;[1 0.5502 0] Multicast Effects
  (Components|Activation|SetActive _damageparticlesystem true false)
  (Effects|Components|ParticleSystem|SetEmitterEnable _damageparticlesystem "Sparks" true))

(event Custom|Activate%20Objective%20Trail
  (params (Instigator :type Controller%20Object%20Reference))
  ;[0.2393 0 1 0.25] Gameplay Functions
  ;;[0.2393 0 1 0.25] ObjectiveTrail
  (Variables|Default|SetObjectiveTrailInstigator Instigator)
  (Variables|Default|SetActivateObjectiveTrail true)
  (Class|CreativeObjectiveTrailComponent|ToggleObjectiveTrail (Variables|Default|GetCreative_ObjectiveTrail_Component) (Variables|Default|GetActivateObjectiveTrail) self (Variables|Default|GetObjectiveTrailInstigator) (Creative|TeamOption|GetLegacyTeamIndex (Variables|UserOptions|Requirements|GetOwning%20Team)) (Variables|UserOptions|ObjectivePulse|GetShow%20Objective%20Pulse%20to%20Friendly%20Players) (Variables|UserOptions|ObjectivePulse|GetShow%20Objective%20Pulse%20to%20Enemy%20Players) (Variables|UserOptions|ObjectivePulse|GetShow%20Objective%20Pulse%20to%20Instigator%20Only) false 2.0 0))

(event Custom|Deactivate%20Objective%20Trail
  (params (Instigator :type Controller%20Object%20Reference))
  ;[0.2393 0 1 0.25] Gameplay Functions
  ;;[0.2393 0 1 0.25] ObjectiveTrail
  (Variables|Default|SetObjectiveTrailInstigator Instigator)
  (Variables|Default|SetActivateObjectiveTrail false)
  (Class|CreativeObjectiveTrailComponent|ToggleObjectiveTrail (Variables|Default|GetCreative_ObjectiveTrail_Component) (Variables|Default|GetActivateObjectiveTrail) self (Variables|Default|GetObjectiveTrailInstigator) (Creative|TeamOption|GetLegacyTeamIndex (Variables|UserOptions|Requirements|GetOwning%20Team)) (Variables|UserOptions|ObjectivePulse|GetShow%20Objective%20Pulse%20to%20Friendly%20Players) (Variables|UserOptions|ObjectivePulse|GetShow%20Objective%20Pulse%20to%20Enemy%20Players) (Variables|UserOptions|ObjectivePulse|GetShow%20Objective%20Pulse%20to%20Instigator%20Only) false 2.0 0))

(event Custom|PickMesh
  ;[1 0.5502 0] Mesh Selecftion
  (Development|PrintString (Utilities|String|Append "Pick: Mesh is  " (Utilities|String|EnumtoString (Variables|UserOptions|Basic|GetMesh))))
  (bind (_custommesh _stat_tag) (Utilities|GetMeshDetails)
    (:%20HasDetails
      (Utilities|FlowControl|Sequence
        (:then_0
          (bind _loaded_object (Creative|Loading|AsyncLoadEditorSafe _custommesh)
            (:Loaded
              (Utilities|ValidateAsyncLoadedObject _custommesh _loaded_object
                (:true
                  (bind _asstatic_mesh (Utilities|Casting|CastToStaticMesh _loaded_object))
                  (Components|StaticMesh|SetStaticMesh (Variables|Debug|GetStaticMeshComponent) _asstatic_mesh)
                  (Default|SetDynamicMemory (Variables|Default|GetCreativeDynamicMemoryComponent) (Utilities|Array|MakeArray _asstatic_mesh))
                  (CallFunction|SetDamageParticlesRelativeLocation)
                  (CallFunction|UpdateHealthbarHeight)
                  (for (_ _array_index) (Rendering|Material|GetMaterials (Variables|Debug|GetStaticMeshComponent))
                    (Rendering|Material|SetMaterial (Variables|Debug|GetStaticMeshComponent) _array_index 0) ; Reset all of the materials on the mesh\nto the default.
                  ))))))
        (:then_1
          (Variables|Default|SetDestructionTag _stat_tag))))))

(event Custom|SetInvulnerableEvent
  (params (Instigator :type Controller%20Object%20Reference))
  ;[0.2393 0 1 0.25] Gameplay Functions
  (Variables|Default|SetInvulnerableRuntime true))

(event Custom|SetDamagableEvent
  (params (Instigator :type Controller%20Object%20Reference))
  ;[0.2393 0 1 0.25] Gameplay Functions
  (Variables|Default|SetInvulnerableRuntime false))

(event Custom|ResetHealthEvent
  (bind _health (Variables|UserOptions|Health|GetHealth))
  (bind _returnvalue (Math|Float|Max(Float) _health 1.0))
  ;[0.2393 0 1 0.25] Gameplay Functions
  (if (!= (Health|GetHealth) _returnvalue)
    (Gameplay|ForceBuildingHealth _returnvalue)
    (Networking|ForceNetUpdate)
    (CallFunction|RPCMulticastResetHealthEffects)))

(event MinigameStartedDelegate(FortMinigameObjective)
  (Variables|Default|SetInvulnerableRuntime (Variables|UserOptions|Health|GetInvulnerable)) ; Determines if the device can be damaged.\nThis is separate so it can be altered during game without\ncomprimising the saved data.
  (Variables|Default|SetCachedHealth (Health|GetHealth))
  (Networking|SwitchHasAuthority
    (:Authority
      (CallFunction|UpdateTrackingState (Variables|/Script/FortniteGame.FortMinigameLogicComponent|GetCurrentMinigame (Variables|Default|GetFortMinigameObjective)))
      (CallFunction|ResetDestructionLimit))))

(event Custom|UpdateTeam
  (bind _returnvalue (Utilities|GetTeamfromTeamSetting))
  (if (!= (Math|Conversions|ToInteger(Byte) (Variables|Team|GetTeamIndex)) _returnvalue)
    (CallFunction|SetTeam (Math|Conversions|ToByte(Integer) _returnvalue))
    (else
      _))
  (CallFunction|SetParticle))

(event MinigameEndedDelegate(FortMinigameObjective)
  (Variables|Default|SetInvulnerableRuntime (Variables|UserOptions|Health|GetInvulnerable)) ; Determines if the device can be damaged.\nThis is separate so it can be altered during game without\ncomprimising the saved data.
  (Variables|Default|SetCachedHealth (Health|GetHealth))
  (Networking|SwitchHasAuthority
    (:Authority
      (CallFunction|UpdateTrackingState (Variables|/Script/FortniteGame.FortMinigameLogicComponent|GetCurrentMinigame (Variables|Default|GetFortMinigameObjective)))
      (CallFunction|ResetDestructionLimit))))

(event EventPostFinishSpawning
  (|CallParentFunction)
  (Utilities|FlowControl|Sequence
    (:then_0
      (CallFunction|TeamSettingUpgradePath)
      (CallFunction|UpdateTeam))
    (:then_1
      (Class|FortMinigameLogicComponent|TryAttachtoMinigame (Variables|Default|GetFortMinigameObjective)))))

(event Custom|Notify%20Feedback%20Event
  (params (EventParams :type Gameplay%20Feedback%20Event%20Params%20Structure) (ProgressPct :type Float%20%28double-precision%29))
  (bind (_eventtag _instigatorplayerid) (Utilities|Struct|BreakGameplayFeedbackEventParams EventParams))
  (CallFunction|PostObjectiveFeedbackEvent _eventtag _instigatorplayerid ProgressPct))

(event OnAddedtoMinigame(FortMinigameObjective)
  (params (Minigame :type Fort%20Minigame%20Object%20Reference))
  (Networking|SwitchHasAuthority
    (:Authority
      (CallFunction|UpdateTrackingState Minigame))
    (:Remote
      (Minigame|BindEventtoPlayerAddedDelegate Minigame (EventDispatchers|CreateEvent :!event Custom|OnPlayerAdded))
      (Minigame|BindEventtoPlayerRemovedDelegate Minigame (EventDispatchers|CreateEvent :!event Custom|OnPlayerRemoved))
      (CallFunction|BindAllCurrentMinigamePlayers))))

(event Custom|SetParticle
  (bind _beacon (Variables|UserOptions|UI|GetBeacon))
  (bind _beaconparticletypes (Variables|Default|GetBeaconParticleTypes))
  (bind _value (Utilities|Map|Find _beaconparticletypes _beacon))
  (bind (_particletype_3_45bbedad49901d6a2ff9fab0205b3f51 _colorparameters_13_112d79bd4c25b6efa3483f992bd13189) (Utilities|Struct|BreakFDestructionObjectParticleType _value))
  (bind _particlesystem (Variables|Default|GetParticleSystem))
  (bind _particlesystem_2 (Variables|Default|GetParticleSystem))
  (Utilities|IsValid _particlesystem_2
    (:Is%20Valid
      (Components|Activation|SetActive _particlesystem_2 false true))
    (:Is%20Not%20Valid
      _))
  (if (Utilities|Enum|NotEqual(Enum) _beacon "NewEnumerator0")
    (bind _loaded_object (Creative|Loading|AsyncLoadEditorSafe _particletype_3_45bbedad49901d6a2ff9fab0205b3f51)
      (:Loaded
        (Utilities|ValidateAsyncLoadedObject _particletype_3_45bbedad49901d6a2ff9fab0205b3f51 _loaded_object
          (:true
            (bind _asparticle_system (Utilities|Casting|CastToParticleSystem _loaded_object))
            (Effects|Components|ParticleSystem|SetTemplate (Variables|Default|GetParticleSystem) _asparticle_system)
            (Components|Activation|SetActive _particlesystem true true)
            (CallFunction|SetBeaconColor _particlesystem _colorparameters_13_112d79bd4c25b6efa3483f992bd13189)
            (CallFunction|UpdateBeaconLocation)
            (CallFunction|SetBeaconScale _particlesystem)
            (if (Utilities|Enum|Equal(Enum) (Variables|UserOptions|UI|GetBeacon) "NewEnumerator1")
              (Effects|Components|ParticleSystem|SetEmitterEnable _particlesystem "Letter" false)
              (CallFunction|SetDamageParticlesRelativeLocation))))))))

(event OnRemovedfromMinigame(FortMinigameObjective)
  (params (Minigame :type Fort%20Minigame%20Object%20Reference))
  (Networking|SwitchHasAuthority
    (:Authority
      (Utilities|IsValid Minigame
        (:Is%20Valid
          (Minigame|Objective|StopTracking (Variables|Default|GetFortMinigameObjective) false (Health|GetHealthPercent)))))
    (:Remote
      (Minigame|UnbindEventfromPlayerAddedDelegate Minigame (EventDispatchers|CreateEvent :!event Custom|OnPlayerAdded))
      (Minigame|UnbindEventfromPlayerRemovedDelegate Minigame (EventDispatchers|CreateEvent :!event Custom|OnPlayerRemoved))
      (CallFunction|UnbindAllCurrentMinigamePlayers))))

(event Custom|RPC_Multicast%20Reset%20Health%20Effects
  ;[1 0.5502 0] Force a UI Update on All Players (with forced max health, because the actual health value won't be replicated yet) and disable damage effects
  (Networking|CanExecuteCosmeticEvents
    (:True
      (Default|CallUpdateRefresh true)
      (CallFunction|StopHealthEffects))))

(event Custom|OnPlayerAdded
  (params (UniqueNetId :type Unique%20Net%20Id%20Structure) (bIsLocalPlayer :type Boolean))
  (Networking|SwitchHasAuthority
    (:Remote
      (bind _returnvalue (UniqueNetID|IsValidNetID UniqueNetId))
      (if (and _returnvalue bIsLocalPlayer)
        (CallFunction|BindPlayer UniqueNetId)))))

(event Custom|OnPlayerRemoved
  (params (UniqueNetId :type Unique%20Net%20Id%20Structure) (bIsLocalPlayer :type Boolean))
  (Networking|SwitchHasAuthority
    (:Remote
      (bind _returnvalue (UniqueNetID|IsValidNetID UniqueNetId))
      (if (and _returnvalue bIsLocalPlayer)
        (CallFunction|UnbindPlayer UniqueNetId)))))

(event EventOnDamageServer
  (params (Damage :type Float%20%28single-precision%29) (DamageTags :type Gameplay%20Tag%20Container%20Structure%20%28by%20ref%29) (Momentum :type Vector :default "0, 0, 0") (HitInfo :type Hit%20Result%20Structure%20%28by%20ref%29) (InstigatedBy :type Controller%20Object%20Reference) (DamageCauser :type Actor%20Object%20Reference) (EffectContext :type Gameplay%20Effect%20Context%20Handle%20Structure))
  (bind _returnvalue (Health|GetHealth))
  (Utilities|FlowControl|Sequence
    (:then_0
      (CallFunction|UpdateTrackedProgress)
      (if (< _returnvalue (Variables|Default|GetCachedHealth))
        (CallFunction|BroadcastDamageEvent InstigatedBy Damage)
        (if (CallFunction|ShouldShowFX)
          (CallFunction|UpdateHealthEffects Damage))))
    (:then_1
      (Variables|Default|SetCachedHealth _returnvalue))))

(event Custom|Destroy%20Event
  (params (Instigator :type Controller%20Object%20Reference))
  ;[0.274 0 0.001878] Damage and Destruction
  (Variables|Default|SetDestroyInstigator Instigator)
  (if (not (Variables|BuildingActor|GetbDestroyed))
    (CallFunction|DoDestroy)
    (CallFunction|HandleHealthChangeEvents 0.0 (Variables|Default|GetDestroyInstigator) true)
    (Minigame|Objective|SetTrackedProgress (Variables|Default|GetFortMinigameObjective) 0.0)
    (Utilities|FlowControl|Sequence
      (:then_0
        (Utilities|FlowControl|DoOnce false
          (:Completed
            (GameplayEvent|BroadcastEvent (Variables|UserOptions|Events|GetOn%20Destroyed) 0 (Utilities|Struct|MakeGameplayEventDestroyed (Variables|Default|GetDestroyInstigator))))) ; Stop destruction event firing multiple times
      )
      (:then_1
        (Actor|DestroyActor))) ; Broadcast has to happen before DestroyActor\nor Verse events that are bound will fail to trigger\n(since the reference will be invalid)
  ))

(event Custom|InstigateDestroy
  (params (Instigator :type Controller%20Object%20Reference))
  ;[0.274 0 0.001878] Damage and Destruction
  (bind _output_get (Variables|Default|SetDestroyInstigator Instigator))
  (if (or (Math|Byte|NotEqual(Byte) (Team|GetActorTeam _output_get) (Team|GetActorTeam self)) (Creative|TeamOption|IsTeamType (Variables|UserOptions|Requirements|GetOwning%20Team) "None"))
    (if (not (Variables|BuildingActor|GetbDestroyed))
      (CallFunction|DoDestroy)
      (CallFunction|HandleHealthChangeEvents 0.0 (Variables|Default|GetDestroyInstigator) true)
      (Minigame|Objective|SetTrackedProgress (Variables|Default|GetFortMinigameObjective) 0.0)
      (Utilities|FlowControl|Sequence
        (:then_0
          (Utilities|FlowControl|DoOnce false
            (:Completed
              (GameplayEvent|BroadcastEvent (Variables|UserOptions|Events|GetOn%20Destroyed) 0 (Utilities|Struct|MakeGameplayEventDestroyed (Variables|Default|GetDestroyInstigator))))) ; Stop destruction event firing multiple times
        )
        (:then_1
          (Actor|DestroyActor))) ; Broadcast has to happen before DestroyActor\nor Verse events that are bound will fail to trigger\n(since the reference will be invalid)
    )))

(event EventOnDeathServer
  (params (Damage :type Float%20%28single-precision%29) (DamageTags :type Gameplay%20Tag%20Container%20Structure%20%28by%20ref%29) (Momentum :type Vector :default "0, 0, 0") (HitInfo :type Hit%20Result%20Structure%20%28by%20ref%29) (InstigatedBy :type Controller%20Object%20Reference) (DamageCauser :type Actor%20Object%20Reference) (EffectContext :type Gameplay%20Effect%20Context%20Handle%20Structure))
  ;[0.274 0 0.001878] Damage and Destruction
  (Variables|Default|SetDestroyInstigator InstigatedBy)
  (CallFunction|DoDestroy)
  (CallFunction|HandleHealthChangeEvents 0.0 (Variables|Default|GetDestroyInstigator) true)
  (CallFunction|SetVisibilityForcedDuringGame "NewEnumerator2")
  (CallFunction|SetCollisionForcedDuringGame "NewEnumerator2")
  (Utilities|FlowControl|DoOnce false
    (:Completed
      (GameplayEvent|BroadcastEvent (Variables|UserOptions|Events|GetOn%20Destroyed) 0 (Utilities|Struct|MakeGameplayEventDestroyed (Variables|Default|GetDestroyInstigator))))) ; Stop destruction event firing multiple times
)

(event Custom|Reset%20Destruction%20Limit
  ;[0.274 0 0.001878] Damage and Destruction
  (Utilities|FlowControl|DoOnce false
    (:Completed
      (GameplayEvent|BroadcastEvent (Variables|UserOptions|Events|GetOn%20Destroyed) 0 (Utilities|Struct|MakeGameplayEventDestroyed (Variables|Default|GetDestroyInstigator))))) ; Stop destruction event firing multiple times
)

(event Custom|DestroySurroundingActors
  ;[0.274 0 0.001878] Damage and Destruction
  (bind _returnvalue (Utility|OverlapQueue|GetAsyncOverlapQueue self))
  (Utilities|IsValid _returnvalue
    (:Is%20Valid
      (Class|AsyncOverlapQueue|ThrottledAsyncSphereOverlapActors :self _returnvalue :Location (Transformation|GetActorLocation) :Radius (* (Variables|UserOptions|Basic|GetBlast%20Radius) 1e+02) :bCheckLineOfSight false :ShouldCancelProcess (EventDispatchers|CreateEvent :!event ShouldCancelDestruction) :ProcessHitActorFunction (delegate Custom|DestroyHitActor) :CompletionFunction (delegate Custom|OverlapFinished) :ObjectTypes "ObjectTypeQuery1" :TraceChannel "ECC_Visibility")
      ;;[1 1 1] Wait for our destruction to be done.
      (Actor|SetLifeSpan 0.0))))

(event Custom|DestroyHitActor
  (params (HitActor :type Actor%20Object%20Reference))
  ;[0.274 0 0.001878] Damage and Destruction
  (bind _asbuilding_actor (Utilities|Casting|CastToBuildingActor HitActor))
  (if (Variables|/Script/Engine.Actor|GetbCanBeDamaged _asbuilding_actor)
    (bind _asdevice_objective_v2 (Utilities|Casting|CastToDevice_Objective_V2 _asbuilding_actor)
      (:then
        (CallFunction|InstigateDestroy _asdevice_objective_v2 (Variables|Default|GetDestroyInstigator)))
      (:CastFailed
        (Actor|DestroyActor _asbuilding_actor)))))

(event Custom|OverlapFinished
  (params (WasCancelled :type Boolean))
  (bind _death_sound (Variables|Default|GetDeath%20Sound))
  ;[0.274 0 0.001878] Damage and Destruction
  (if (Audio|Components|Audio|IsPlaying _death_sound)
    (EventDispatchers|BindEventtoOnAudioFinished _death_sound (delegate Custom|OnAudioFinished_Event))
    (else
      (Actor|SetLifeSpan 0.01))))

(event Custom|OnAudioFinished_Event
  ;[0.274 0 0.001878] Damage and Destruction
  (Actor|SetLifeSpan 0.01))

(event EventOnDeathPlayEffects
  (params (Damage :type Float%20%28single-precision%29) (DamageTags :type Gameplay%20Tag%20Container%20Structure%20%28by%20ref%29) (Momentum :type Vector :default "0, 0, 0") (HitInfo :type Hit%20Result%20Structure%20%28by%20ref%29) (InstigatedBy :type Fort%20Pawn%20Object%20Reference) (DamageCauser :type Actor%20Object%20Reference) (EffectContext :type Gameplay%20Effect%20Context%20Handle%20Structure))
  ;[0.274 0 0.001878] Damage and Destruction
  (switch Utilities|FlowControl|Switch|SwitchonEnum_PlayAudioVFXSetting (Variables|UserOptions|GetPlay%20Audio%20/%20VFX)
    (:NewEnumerator1
      (Audio|Components|Audio|Play (Variables|Default|GetDeath%20Sound) 0.0)
      (Networking|CanExecuteCosmeticEvents
        (:True
          (if (Variables|UserOptions|Basic|GetBlast%20Radius_Override)
            (Effects|SpawnEmitteratLocation :EmitterTemplate "/Game/Creative/Devices/ExplosiveBarrel/P_ExplosiveBarrel_ExplosionMain.P_ExplosiveBarrel_ExplosionMain" :Location (Transformation|GetWorldLocation (Variables|Debug|GetStaticMeshComponent)) :Scale (CallFunction|GetExplosionFXScale) :bAutoDestroy true :PoolingMethod "AutoRelease" :bAutoActivateSystem true)))))))

(event EventOnDamagePlayEffects
  (params (Damage :type Float%20%28single-precision%29) (DamageTags :type Gameplay%20Tag%20Container%20Structure%20%28by%20ref%29) (Momentum :type Vector :default "0, 0, 0") (HitInfo :type Hit%20Result%20Structure%20%28by%20ref%29) (InstigatedBy :type Fort%20Pawn%20Object%20Reference) (DamageCauser :type Actor%20Object%20Reference) (EffectContext :type Gameplay%20Effect%20Context%20Handle%20Structure))
  ;[0.274 0 0.001878] Damage and Destruction
  (Variables|Default|SetLastDamageEventTime (Utilities|Time|GetGameTimeinSeconds))
  (if (Utilities|Enum|Equal(Enum) (Variables|UserOptions|GetPlay%20Audio%20/%20VFX) "NewEnumerator1")
    (Audio|Components|Audio|Play (Variables|Default|GetImpact%20Sound) 0.0)
    (if (not (Variables|Default|GetBadgeIsTicking))
      (Utilities|FlowControl|RetriggerableDelay (Variables|UserOptions|UI|GetShow%20On%20Damage%20Duration)) ; If the badge isn't periodically updating itself,\nwe need to manually refresh after we should finish.
      (Default|CallUpdateRefresh false))))

(event HandleVersionUpgrades(CreativeVersion)
  (params (LastSavedVersion :type Creative%20Version%20Id%20Structure))
  (CreativeVersion|CompareVersiontoTag LastSavedVersion "(TagName=\"Creative.Version.Release.29.20\")"
    (:OlderVersion
      (Variables|UserOptions|Basic|SetMesh (Utilities|Enum|BytetoEnumEObjectiveMeshes (Math|Conversions|ToByte(Integer) (Variables|Legacy|GetMeshIndex))))
      (Variables|Legacy|SetMeshIndex 0)
      (OptionData|RefreshPropertyOverrides (Variables|Default|GetToyOptionsComponent))
      (CallFunction|PickMesh))))

(fn ActivateFromOtherDevice
  (params (Instigator :type Controller%20Object%20Reference) (Requester :type ERequesterType%20Enum :default "NewEnumerator0"))
  (return false))

(fn BindAllCurrentMinigamePlayers
  (bind _fortminigameobjective (Variables|Default|GetFortMinigameObjective))
  (bind _currentminigame (Variables|/Script/FortniteGame.FortMinigameLogicComponent|GetCurrentMinigame _fortminigameobjective))
  (Utilities|IsValid _currentminigame
    (:Is%20Valid
      (bind _output_get (LocalVariables|SetParticipatingPlayers (Minigame|GetParticipatingPlayers _currentminigame)))
      (for _array_element _output_get
        (bind _asfort_player_controller_athena (Utilities|Casting|CastToFortPlayerControllerAthena (Fortnite|GetFortPlayerControllerfromActor _array_element)))
        (bind _output_get_2 (LocalVariables|SetCurrentParticipatingPlayer _asfort_player_controller_athena))
        (if (Pawn|IsLocalPlayerController _output_get_2)
          (FortPlayerControllerAthena|BindEventtoOnLocalPlayerChangedTeams _output_get_2 (EventDispatchers|CreateEvent :!event UpdateFromTeamChanged)))))))

(fn BindPlayer
  (params (UniqueNetID :type Unique%20Net%20Id%20Structure%20%28by%20ref%29))
  (bind _output_get (LocalVariables|SetLocalPlayerControllers (Fortnite|GetLocalFortPlayerControllers true)))
  (bind _array_element (Utilities|Array|ForEachLoopwithBreak _output_get)
    (:LoopBody
      (bind _asfort_player_controller_athena (Utilities|Casting|CastToFortPlayerControllerAthena _array_element))
      (bind _output_get_2 (LocalVariables|SetCurrentLocalPlayer _asfort_player_controller_athena))
      (if (Online|UniqueNetIdRepl|Equal(UniqueIdRepl) (Default|GetGameAccountId _output_get_2) (Variables|GetUniqueNetID))
        (FortPlayerControllerAthena|BindEventtoOnLocalPlayerChangedTeams (LocalVariables|GetCurrentLocalPlayer) (EventDispatchers|CreateEvent :!event UpdateFromTeamChanged))))))

(fn BlueprintModifyIncomingDamage
  (params (Damage :type Float%20%28single-precision%29) (InTags :type Gameplay%20Tag%20Container%20Structure%20%28by%20ref%29) (EffectContext :type Gameplay%20Effect%20Context%20Handle%20Structure%20%28by%20ref%29) (EventInstigator :type Controller%20Object%20Reference) (DamageCauser :type Actor%20Object%20Reference))
  (bind _currentminigame (Variables|/Script/FortniteGame.FortMinigameLogicComponent|GetCurrentMinigame (Variables|Default|GetFortMinigameObjective))
    (:then
      (if (Math|Boolean|ANDBoolean (Minigame|IsinProgress _currentminigame) (Math|Byte|NotEqual(Byte) (Team|GetActorTeam EventInstigator) (Variables|Team|GetTeamIndex)) (or (CallFunction|CanTakeDamage) (GameplayTags|HasTag InTags (LocalVariables|Getlocal%20Indestructible%20Damage%20Tag) true)))
        (return Damage)
        (else
          _)))
    (:else
      _))
  (return 0.0))

(fn BroadcastDamageEvent
  (params (Controller :type Controller%20Object%20Reference) (Damage :type Float%20%28double-precision%29))
  (bind _returnvalue (Utilities|Time|GetGameTimeinSeconds))
  (if (>= _returnvalue (+ (Variables|Default|GetLastDamageEventTime) (Variables|UserOptions|GetDamage%20Event%20Cooldown)))
    (Variables|Default|SetLastDamageEventTime _returnvalue)
    (GameplayEvent|BroadcastEvent (Variables|UserOptions|Events|GetOn%20Damaged) 0 (Utilities|Struct|MakeGameplayEventonDamaged Controller Damage))))

(fn CanInteract
  (params (InteractingPawn :type Fort%20Pawn%20Object%20Reference) (InteractionBeingAttempted :type EInteractionBeingAttempted%20Enum :default "FirstInteraction") (InteractionType :type TInteractionType%20Enum :default "IT_NoInteraction"))
  (return false))

(fn CanTakeDamage
  (return (not (Variables|Default|GetInvulnerableRuntime))))

(fn DoDestroy
  (bind _currentminigame (Variables|/Script/FortniteGame.FortMinigameLogicComponent|GetCurrentMinigame (Variables|Default|GetFortMinigameObjective))
    (:then
      (Minigame|IsInProgress _currentminigame
        (:Yes
          (CallFunction|SendDestroyedEvent)
          (if (Variables|UserOptions|Basic|GetBlast%20Radius_Override)
            (Environment|ApplyGlobalEnvironmentGameplayEffecttoActor self "/Game/Creative/Devices/Destruction_Object/GE_Creative_DestructionObject.GE_Creative_DestructionObject_C")
            (CallFunction|DestroySurroundingActors)))))
    (:else)))

(fn GetDeviceColor
  (bind _output (CallFunction|GetBeaconColor))
  (return (Math|Conversions|ToLinearColor(Vector) _output)))

(fn GetInstigatingController
  (return 0))

(fn GetBeaconColor
  (if (Utilities|Enum|NotEqual(Enum) (Variables|UserOptions|UI|GetBeacon%20Color) "NewEnumerator3")
    (return (Math|Conversions|ToVector(LinearColor) (Variables|UserOptions|UI|GetCustom%20Beacon%20Color)) true)
    (else
      (return (Math|Conversions|ToVector(LinearColor) (Utilities|Select :Index (select (Editor|InEditor) (select (Utilities|Enum|Equal(Enum) (Creative|TeamOption|ConverttoTeam (Variables|UserOptions|Requirements|GetOwning%20Team)) "HumanPvP_Team1") "Friendly" "Hostile") (Team|GetTeamAffiliationtoLocalPlayer (Math|Conversions|ToByte(Integer) (Utilities|GetTeamfromTeamSetting)) true true)) :Friendly (Variables|Default|GetFriendlyTeamBeaconColor) :Neutral (Variables|Default|GetFriendlyTeamBeaconColor) :Hostile (Variables|Default|GetEnemyTeamBeaconColor))) true))))

(fn GetExplosionFXScale
  ;[1 0.8268 0 0.25] Explosion FX works in 'Tiles' range between 0.25 and 2, Blast Radius works in meters. This does converts between.
  (return (Math|Conversions|ToVector(Float) (Math|Float|Clamp(Float) (/ (Variables|UserOptions|Basic|GetBlast%20Radius) 5.12) 0.25 2.0))))

(fn GetLocalHideDistance
  (return (* (Variables|UserOptions|UI|GetBadge%20Visibility%20Distance) 1e+02)))

(fn GetObjectiveNameStyle
  (params (EventTag :type Gameplay%20Tag%20Structure))
  (bind _teamindex (Variables|Team|GetTeamIndex))
  (bind _returnvalue (Team|GetTeamAffiliationtoLocalPlayer _teamindex true true))
  (switch Utilities|FlowControl|Switch|SwitchonGameplayTag EventTag
    (:Default
      _)
    (:Creative.Event.Objective.Warning
      (return (Utilities|Select :Index _returnvalue :Friendly "Ally" :Neutral "Default" :Hostile "Enemy")))
    (:Creative.Event.Objective.Critical
      _)
    (:Creative.Event.Objective.Destroyed
      _))
  (return (Utilities|Select :Index _returnvalue :Friendly "TeamVictim" :Neutral "Victim" :Hostile "Victim")))

(fn GetOverrideMeshMaterial
  (params (MeshComp :type Static%20Mesh%20Component%20Object%20Reference) (MatIdx :type Integer))
  (return (Rendering|Material|GetMaterial MeshComp MatIdx)))

(fn GetSpawnedPickups
  (return 0))

(fn GetUIHideCondition
  (Utilities|FlowControl|Sequence
    (:then_0
      ;[0 1 0.9868 0.539] If we want to show on Target, ensure this device is the targeted building (GetTargetedBuilding returns if you're highlighting the current building, e.g. within range of your current weapon).
      (Utilities|ValidTeamforLocalPlayer (Variables|UserOptions|UI|GetShowBadgeOnTarget) (Variables|Default|GetLocalPlayerFriendly)
        (:True
          (if (== (Class|FortPlayerController|GetTargetedBuilding (Fortnite|GetFirstLocalFortPlayerController true)) self)
            (return "NewEnumerator0")))))
    (:then_1
      ;[0 1 0.9564 0.204] If we want to show on damage, test to ensure that damage was done within the timeframe.
      (Utilities|ValidTeamforLocalPlayer (Variables|UserOptions|UI|GetShowBadgeOnDamage) (Variables|Default|GetLocalPlayerFriendly)
        (:True
          (if (< (Utilities|Time|GetGameTimeinSeconds) (+ (Variables|Default|GetLastDamageEventTime) (Variables|UserOptions|UI|GetShow%20On%20Damage%20Duration)))
            (return "NewEnumerator0")))))
    (:then_2
      (bind _output_get (LocalVariables|SetLocalVisibility (select (Variables|Default|GetLocalPlayerFriendly) (Variables|UserOptions|UI|GetFriendlyBadgeVisibility) (Variables|UserOptions|UI|GetHostileBadgeVisibility))))
      (return (Utilities|Select :Index _output_get :NewEnumerator0 "NewEnumerator1" :NewEnumerator1 "NewEnumerator2" :NewEnumerator2 "NewEnumerator0")))))

(fn GetWidgetOwnerLocation
  (params (WidgetComponent :type Widget%20Component%20Object%20Reference))
  (Utilities|FlowControl|Sequence
    (:then_0
      (Utilities|IsValid WidgetComponent
        (:Is%20Valid
          (bind _output_get (LocalVariables|SetWidgetObject (UserInterface|GetUserWidgetObject (Variables|GetWidgetComponent))))
          (Utilities|IsValid _output_get
            (:Is%20Valid
              (bind _output_get_2 (LocalVariables|SetPlayer (Widget|GetOwningPlayer _output_get)))
              (Utilities|IsValid _output_get_2
                (:Is%20Valid
                  (bind _output_get_3 (LocalVariables|SetControlledPawn (Pawn|GetControlledPawn _output_get_2)))
                  (Utilities|IsValid _output_get_3
                    (:Is%20Valid
                      (LocalVariables|SetPlayerLocation (Transformation|GetActorLocation _output_get_3)))))))))))
    (:then_1
      (return (LocalVariables|GetPlayerLocation)))))

(fn HandleHealthChangeEvents
  (params (HealthPct :type Float%20%28double-precision%29) (Instigator :type Controller%20Object%20Reference) (bProcessDestroy :type Boolean))
  (bind _criticalnotificationthreshold (Variables|UserOptions|UI|GetCriticalNotificationThreshold))
  (bind _returnvalue (/ _criticalnotificationthreshold 1e+02))
  (bind _warningnotificationthreshold (Variables|UserOptions|UI|GetWarningNotificationThreshold))
  (bind _returnvalue_2 (/ _warningnotificationthreshold 1e+02))
  (Utilities|FlowControl|Sequence
    (:then_0
      (LocalVariables|SetPrevProgressPct (Variables|/Script/FortniteGame.FortMinigameObjectiveComponent|GetProgressPercentage (Variables|Default|GetFortMinigameObjective)))
      (bind _instigator (Variables|Actor|GetInstigator)
        (:then
          (bind _asfort_player_state (Utilities|Casting|CastToFortPlayerState (Variables|/Script/Engine.Controller|GetPlayerState _instigator)))
          (LocalVariables|SetInstigatorPlayerId (PlayerState|GetUniqueID _asfort_player_state)))
        (:else)))
    (:then_1
      (if (Math|Boolean|ANDBoolean (Variables|UserOptions|UI|GetShow%20Destroy%20Messages) bProcessDestroy (<= (Variables|GetHealthPct) 0.0))
        (CallFunction|SendObjectiveFeedbackEvent "(TagName=\"Creative.Event.Objective.Destroyed\")" (LocalVariables|GetInstigatorPlayerId))
        (elif (Math|Boolean|ANDBoolean (> (LocalVariables|GetPrevProgressPct) _returnvalue) (<= (Variables|GetHealthPct) _returnvalue) (Variables|UserOptions|UI|GetCriticalNotificationThreshold_Override))
          (CallFunction|SendObjectiveFeedbackEvent "(TagName=\"Creative.Event.Objective.Critical\")" (LocalVariables|GetInstigatorPlayerId) _returnvalue)
          (elif (Math|Boolean|ANDBoolean (> (LocalVariables|GetPrevProgressPct) _returnvalue_2) (<= (Variables|GetHealthPct) _returnvalue_2) (Variables|UserOptions|UI|GetWarningNotificationThreshold_Override))
            (CallFunction|SendObjectiveFeedbackEvent "(TagName=\"Creative.Event.Objective.Warning\")" (LocalVariables|GetInstigatorPlayerId) _returnvalue_2)))))))

(fn IsEnabled
  (return false))

(fn IsHoldingItemForPawn
  (params (Item :type Item%20Variant%20Handle%20Structure) (Pawn :type Pawn%20Object%20Reference))
  (return false))

(fn IsInPlayMode
  (bind _fortminigamelogic (Variables|Default|GetFortMinigameLogic)
    (:then
      (bind _currentminigame (Variables|/Script/FortniteGame.FortMinigameLogicComponent|GetCurrentMinigame _fortminigamelogic)
        (:then
          (return (Minigame|IsinPlayMode _currentminigame :!class /Script/FortniteGame.FortMinigame)))
        (:else
          _)))
    (:else
      _))
  (return false))

(fn IsTargetableByLinearMover
  (return true))

(fn MessageScore
  (return (Variables|UserOptions|ScoringHUD|GetHUD%20Message) (Variables|UserOptions|ScoringHUD|GetHUD%20MessageScoreColor) (Variables|UserOptions|ScoringHUD|GetHUD%20MessageColor) (Variables|UserOptions|ScoringHUD|GetResetHUD%20MessageScore)))

(fn MigrateLegacyUserOptions
  (bind _legacybeaconscale (Class|GetClassDefaults "/CreativeCoreDevices/Device_Objective_V2.Device_Objective_V2_C"))
  (bind _warningeventpct (Class|GetClassDefaults "/CreativeCoreDevices/Device_Objective_V2.Device_Objective_V2_C"))
  (bind _explosionradius (Class|GetClassDefaults "/CreativeCoreDevices/Device_Objective_V2.Device_Objective_V2_C"))
  (bind _hide_beacon_at (Class|GetClassDefaults "/CreativeCoreDevices/Device_Objective_V2.Device_Objective_V2_C"))
  (bind _beaconparticletype (Class|GetClassDefaults "/CreativeCoreDevices/Device_Objective_V2.Device_Objective_V2_C"))
  (bind (_showbadgeontarget _showbadgeondamage _collisionprofileduringgamesknob) (Class|GetClassDefaults "/CreativeCoreDevices/Device_Objective_V2.Device_Objective_V2_C"))
  (bind _beaconcolorindex (Class|GetClassDefaults "/CreativeCoreDevices/Device_Objective_V2.Device_Objective_V2_C"))
  (bind _criticaleventpct (Class|GetClassDefaults "/CreativeCoreDevices/Device_Objective_V2.Device_Objective_V2_C"))
  (bind _beaconcolorindex_2 (Variables|Legacy|GetBeaconColorIndex))
  (bind _returnvalue (== _beaconcolorindex_2 -1))
  (bind _criticaleventpct_2 (Variables|Legacy|GetCriticalEventPct))
  (bind _warningeventpct_2 (Variables|Legacy|GetWarningEventPct))
  (bind _healthvalue (Variables|Legacy|GetHealthValue))
  (bind _beaconparticletype_2 (Variables|Legacy|GetBeaconParticleType))
  (bind _legacybeaconscale_2 (Variables|Legacy|GetLegacyBeaconScale))
  (bind _collisionprofileduringgamesknob_name (Utilities|Struct|BreakCollisionProfileName (Variables|GetCollisionProfileDuringGamesKnob_Name)))
  (|CallParentFunction)
  (Utilities|FlowControl|Sequence
    (:then_0
      (Utilities|FlowControl|Sequence
        (:then_0
          ;[0.9829 0.4688 1] Migrate Invulnerable and health value
          (if (!= (Class|GetClassDefaults "/CreativeCoreDevices/Device_Objective_V2.Device_Objective_V2_C") _healthvalue)
            (bind _output_get (Variables|UserOptions|Health|SetInvulnerable (<= _healthvalue 0)))
            (Variables|UserOptions|Health|SetHealth (Math|Float|SelectFloat (Class|GetClassDefaults "/CreativeCoreDevices/Device_Objective_V2.Device_Objective_V2_C") _healthvalue _output_get))
            (Variables|Legacy|SetHealthValue _healthvalue)
            (LocalVariables|SetMigratedProperty true)))
        (:then_1
          ;[0.3431 1 0.7003] Migrate Explosion Radius
          (if (!= (Variables|Legacy|GetExplosionRadius) _explosionradius)
            (Variables|UserOptions|Basic|SetBlast%20Radius_Override (<= _explosionradius 0))
            (bind _result (ArrayMacros|SafeArrayGet (Variables|Legacy|GetOldBlastRadiusValues) (Variables|Legacy|GetExplosionRadius))
              (:%20%20
                (Variables|UserOptions|Basic|SetBlast%20Radius _result)
                (Variables|Legacy|SetExplosionRadius _explosionradius)
                (LocalVariables|SetMigratedProperty true)))))
        (:then_2
          ;[0.121 1 0.4041] Migrate Beacon Type
          (if (!= _beaconparticletype _beaconparticletype_2)
            (Variables|UserOptions|UI|SetBeacon (select (!= _beaconparticletype_2 -1) (Utilities|Select :Index _beaconparticletype_2 :Option%200 "NewEnumerator1" :Option%201 "NewEnumerator2" :Option%202 "NewEnumerator3") "NewEnumerator0"))
            (Variables|Legacy|SetBeaconParticleType _beaconparticletype)
            (LocalVariables|SetMigratedProperty true)))
        (:then_3
          ;[1 0.396 0.9766] Migrate Beacon Color
          (if (!= _beaconcolorindex _beaconcolorindex_2)
            (Variables|UserOptions|UI|SetBeacon%20Color (select _returnvalue "NewEnumerator3" "NewEnumerator1"))
            (if (not _returnvalue)
              (Variables|UserOptions|UI|SetCustom%20Beacon%20Color (Utilities|Select :Index (Variables|Legacy|GetBeaconColorIndex) :Option%200 "(R=0.100000,G=0.300000,B=1.000000,A=1.000000)" :Option%201 "(R=1.000000,G=0.000000,B=0.000000,A=1.000000)" :Option%202 "(R=1.000000,G=0.180000,B=0.040000,A=1.000000)"))
              (Variables|Legacy|SetBeaconColorIndex _beaconcolorindex)
              (LocalVariables|SetMigratedProperty true))))
        (:then_4
          ;[0.3591 1 0.9202] Migrate Beacon Scale
          (if (!= _legacybeaconscale_2 _legacybeaconscale)
            (Variables|UserOptions|UI|SetBeacon%20Scale_Override (> _legacybeaconscale_2 0))
            (Variables|UserOptions|UI|SetBeacon%20Scale (Variables|Legacy|GetLegacyBeaconScale))
            (Variables|Legacy|SetLegacyBeaconScale _legacybeaconscale)
            (LocalVariables|SetMigratedProperty true)))
        (:then_5
          ;[1 0.4148 0.315] Migrate Collision
          (if (!= (Utilities|Struct|BreakCollisionProfileName (Class|GetClassDefaults "/CreativeCoreDevices/Device_Objective_V2.Device_Objective_V2_C")) _collisionprofileduringgamesknob_name)
            (Variables|UserOptions|SetCollides%20With "NewEnumerator1")
            (Variables|Legacy|SetCollisionProfileDuringGamesKnob _collisionprofileduringgamesknob)
            (LocalVariables|SetMigratedProperty true)))
        (:then_6
          ;[0.3036 0.7451 1] Migrate Critical
          (if (!= _criticaleventpct_2 _criticaleventpct)
            (Variables|UserOptions|UI|SetCriticalNotificationThreshold_Override false)
            (Variables|UserOptions|UI|SetCriticalNotificationThreshold (* _criticaleventpct_2 1e+02))
            (Variables|Legacy|SetCriticalEventPct _criticaleventpct)
            (LocalVariables|SetMigratedProperty true)))
        (:then_7
          ;[0.2534 0.2081 1] Migrate Warning
          (if (!= _warningeventpct_2 _warningeventpct)
            (Variables|UserOptions|UI|SetWarningNotificationThreshold_Override true)
            (Variables|UserOptions|UI|SetWarningNotificationThreshold (* _warningeventpct_2 1e+02))
            (Variables|Legacy|SetWarningEventPct _warningeventpct)
            (LocalVariables|SetMigratedProperty true)))
        (:then_9
          ;[0.9093 0.3501 1 0.204] Splitting HideBeaconAt property out into separate properties to remove magic numbers (and splitting into friendly and hostile to improve functionality).
          (if (!= _hide_beacon_at (Variables|Legacy|GetHide%20Beacon%20At))
            (if (== (Variables|Legacy|GetHide%20Beacon%20At) -2.0)
              (Variables|UserOptions|UI|SetFriendlyBadgeVisibility "NewEnumerator2")
              (Variables|UserOptions|UI|SetHostileBadgeVisibility "NewEnumerator2")
              (elif (== (Variables|Legacy|GetHide%20Beacon%20At) -1.0)
                (Variables|UserOptions|UI|SetFriendlyBadgeVisibility "NewEnumerator0")
                (Variables|UserOptions|UI|SetHostileBadgeVisibility "NewEnumerator0")
                (else
                  (Variables|UserOptions|UI|SetBadge%20Visibility%20Distance (* (Variables|Legacy|GetHide%20Beacon%20At) 5.12))
                  (Variables|UserOptions|UI|SetFriendlyBadgeVisibility "NewEnumerator1")
                  (Variables|UserOptions|UI|SetHostileBadgeVisibility "NewEnumerator1"))))
            (Variables|Legacy|SetHide%20Beacon%20At _hide_beacon_at)
            (LocalVariables|SetMigratedProperty true)))))
    (:then_1
      (if (LocalVariables|GetMigratedProperty)
        (OptionData|RefreshPropertyOverrides (Variables|Default|GetToyOptionsComponent))))))

(fn OnRep_Beacon
  (CallFunction|SetParticle))

(fn OnRep_BeaconColor
  (CallFunction|SetParticle))

(fn OnRep_BeaconScale
  (CallFunction|SetBeaconScale (Variables|Default|GetParticleSystem)))

(fn OnRep_BeaconScale_Override
  (CallFunction|SetBeaconScale (Variables|Default|GetParticleSystem)))

(fn OnRep_ClampToScreen)

(fn OnRep_CollisionForcedDuringGame
  (Default|UpdateStaticMeshCollision))

(fn OnRep_CustomBeaconColor
  (CallFunction|SetParticle))

(fn OnRep_InvulnerableRuntime
  (Default|CallUpdateRefresh false))

(fn OnRep_OwningTeam
  (CallFunction|UpdatefromTeamChanged))

(fn PostObjectiveFeedbackEvent
  (params (EventTag :type Gameplay%20Tag%20Structure) (InstigatorPlayerId :type Unique%20Net%20Id%20Structure) (ProgressPct :type Float%20%28double-precision%29))
  (bind _feedbackeventmessages (Variables|UI|GetFeedbackEventMessages))
  (bind (_value _returnvalue) (Utilities|Map|Find _feedbackeventmessages EventTag))
  (if _returnvalue
    (FeedbackEvents|PostFeedbackEventfromGameMessage self _value (Variables|GetInstigatorPlayerId) (Utilities|Map|MakeMap "ProgressPct" (Utilities|Text|FormatText "NSLOCTEXT(\"[3B7550A9BAE894F67D7A3D1A802894AD]\", \"8731D41A43597552EE07F7A2EA2C8CF5\", \"{ProgressPct}%\")" (Math|Float|Truncate (* ProgressPct 1e+02))) "ObjectiveStyle" (Utilities|Text|ToText(String) (CallFunction|GetObjectiveNameStyle (Variables|GetEventTag))) "ObjectiveName" (select (Utilities|String|IsEmpty (Utilities|String|ToString(Text) (Variables|UserOptions|UI|GetCustom%20Objective%20Text))) (CallFunction|GetObjectiveNameText (Utilities|Array|FindItem (Variables|Default|GetTextureArrayForIndicator) (Utilities|Struct|BreakCreativeIconOption (Variables|UserOptions|UI|GetObjective%20Identifier)))) (Utilities|Text|FormatText "NSLOCTEXT(\"[3B7550A9BAE894F67D7A3D1A802894AD]\", \"AF772C2548F7942B9E82E084705C12B1\", \"Objective {CustomObjective}\")" (Utilities|String|ToString(Text) (Variables|UserOptions|UI|GetCustom%20Objective%20Text))))))))

(fn SendDestroyedEvent
  (bind (_outsourcetags _outcontexttags) (Class|FortMinigameStatsComponent|GetSourceandContextTags (LocalVariables|Getlocal%20MinigameStatsComp) (Variables|Default|GetDestroyInstigator)))
  (Utilities|IsValid (Variables|Default|GetDestroyInstigator)
    (:Is%20Valid
      (bind _returnvalue (CreativeDevices|GetMinigameStatsComp (Variables|Default|GetDestroyInstigator)))
      (LocalVariables|Setlocal%20MinigameStatsComp _returnvalue)
      (bind _local_minigamestatscomp (LocalVariables|Getlocal%20MinigameStatsComp)
        (:then
          (Class|FortMinigameStatsComponent|SendStatEventwithTags _local_minigamestatscomp (Variables|Default|GetDestroyInstigator) "MinigameDynamicEvent" self (GameplayTags|MakeGameplayTagContainerfromTag (Variables|Default|GetDestructionTag)) (Class|FortMinigameStatsComponent|GetSourceandContextTags _local_minigamestatscomp (Variables|Default|GetDestroyInstigator)) (.OutContextTags (Class|FortMinigameStatsComponent|GetSourceandContextTags _local_minigamestatscomp (Variables|Default|GetDestroyInstigator))) 1)
          (Class|FortMinigameStatsComponent|SendStatEventwithTags (LocalVariables|Getlocal%20MinigameStatsComp) (Variables|Default|GetDestroyInstigator) "MinigameDynamicEvent" self (GameplayTags|MakeGameplayTagContainerfromTag "(TagName=\"Minigame.Stat.Score\")") _outsourcetags _outcontexttags (Variables|UserOptions|GetScore))
          (Class|CreativeHUDBroadcastComponent|DoScoreEventforParticipant (Variables|Default|GetCreative_HUDBroadcast_Component) (Variables|Default|GetDestroyInstigator) (Variables|UserOptions|GetScore) (Variables|UserOptions|ScoringHUD|GetDisplayScoreUpdateOnHUD)))
        (:else)))))

(fn SendObjectiveFeedbackEvent
  (params (EventTag :type Gameplay%20Tag%20Structure) (PlayerId :type Unique%20Net%20Id%20Structure) (ProgressPct :type Float%20%28double-precision%29))
  (CallFunction|NotifyFeedbackEvent (Utilities|Struct|MakeGameplayFeedbackEventParams EventTag PlayerId) ProgressPct))

(fn SetDamageParticlesRelativeLocation
  (bind _fxdamagesocketname (Variables|Default|GetFxDamageSocketName))
  (bind _staticmeshcomponent (Variables|Debug|GetStaticMeshComponent))
  (bind (_returnvalue_location _returnvalue_rotation _returnvalue_scale) (Math|Transform|BreakTransform (Transformation|GetSocketTransform _staticmeshcomponent _fxdamagesocketname "RTS_Component")))
  (bind (_returnvalue_scale_x _returnvalue_scale_y _returnvalue_scale_z) (Math|Vector|BreakVector _returnvalue_scale))
  (bind _returnvalue (Math|Float|Max(Float) _returnvalue_scale_x _returnvalue_scale_y _returnvalue_scale_z))
  (bind _damageparticlesystem (Variables|Default|GetDamageParticleSystem))
  (if (Transformation|DoesSocketExist (Variables|Debug|GetStaticMeshComponent) (Variables|Default|GetFxDamageSocketName))
    (Transformation|SetRelativeLocationAndRotation (Variables|Default|GetDamageParticleSystem) _returnvalue_location _returnvalue_rotation false false)
    (Effects|Components|ParticleSystem|SetFloatParameter _damageparticlesystem "Radius" _returnvalue)
    (Effects|Components|ParticleSystem|SetFloatParameter _damageparticlesystem "SpawnRateScale" (/ (* (Math|Trig|GetPI) (Math|Float|Square _returnvalue)) 7853.97998))))

(fn SetTriggerCollision
  (params (On :type Boolean))
  (return false))

(fn SetBeaconColor
  (params (Beacon :type Cascade%20Particle%20System%20Component%20Object%20Reference) (Array :type Array%20of%20FDestruction%20Object%20Particle%20Color%20Parameter%20Structures))
  (Utilities|IsValid Beacon
    (:Is%20Valid
      (bind (_output _bisvalid) (CallFunction|GetBeaconColor))
      (LocalVariables|SetNewBeaconColor _output)
      (if _bisvalid
        (for _array_element Array
          (Effects|Components|ParticleSystem|SetVectorParameter (Variables|UserOptions|UI|GetBeacon) (Utilities|Struct|BreakFDestructionObjectParticleColorParameter _array_element) (* (LocalVariables|GetNewBeaconColor) (.ColorParameterMultiplier_7_AE2CF29447670F4CE0F86BA17A6BB6D6 (Utilities|Struct|BreakFDestructionObjectParticleColorParameter _array_element)))))))))

(fn SetBeaconScale
  (params (BeaconParticleEffect :type Scene%20Component%20Object%20Reference))
  (Utilities|IsValid BeaconParticleEffect
    (:Is%20Valid
      (if (Variables|UserOptions|UI|GetBeacon%20Scale_Override)
        (Transformation|SetWorldScale3D BeaconParticleEffect (Math|Conversions|ToVector(Float) (Variables|UserOptions|UI|GetBeacon%20Scale)))
        (else
          (Transformation|SetRelativeScale3D BeaconParticleEffect "1.000000,1.000000,1.000000"))))))

(fn SetInvulnerableState
  (params (InvulnerableRuntime :type Boolean))
  (Variables|Default|SetInvulnerableRuntime InvulnerableRuntime))

(fn SetShouldTickHealthBarComponent
  (LocalVariables|SetLocalFriendly (Team|OnSameTeam (Fortnite|GetFirstLocalFortPlayerController true) self))
  (bind _output_get (Variables|Default|SetBadgeIsTicking (Math|Boolean|ORBoolean (Utilities|Enum|Equal(Enum) (select (LocalVariables|GetLocalFriendly) (Variables|UserOptions|UI|GetFriendlyBadgeVisibility) (Variables|UserOptions|UI|GetHostileBadgeVisibility)) "NewEnumerator1") (Utilities|Enum|Equal(Enum) (Variables|UserOptions|UI|GetShowBadgeOnTarget) (select (LocalVariables|GetLocalFriendly) "NewEnumerator1" "NewEnumerator2")) (Utilities|Enum|Equal(Enum) (Variables|UserOptions|UI|GetShowBadgeOnTarget) "NewEnumerator3") (and (Variables|UserOptions|UI|GetClampToScreen) (Variables|UserOptions|UI|GetShowOffscreenArrow)))))
  (Components|Tick|SetComponentTickEnabled (Variables|Default|GetHealthBarComp) _output_get)
  (bind _asobjective_basic_health_bar (Utilities|Casting|CastToObjectiveBasicHealthBar (UserInterface|GetUserWidgetObject (Variables|Default|GetHealthBarComp))))
  (Class|ObjectiveBasicHealthBar|UpdateVisibilityTimer _asobjective_basic_health_bar (Variables|Default|GetBadgeIsTicking)))

(fn ShouldCancelDestruction
  (bind _self self)
  (bind _returnvalue (Minigame|GetMinigame _self :!class /Script/FortniteGame.MinigameBlueprintLibrary))
  (Utilities|IsValid _returnvalue
    (:Is%20Valid
      (return (not (Minigame|IsinProgress _returnvalue))))
    (:Is%20Not%20Valid
      (return true))))

(fn ShouldDisplayDamageNumbers
  (return (Variables|UserOptions|UI|GetDisplay%20Damage%20Numbers)))

(fn ShouldInvisiblePropsBeAllowed
  (return false))

(fn ShouldShowFX
  (return (and (Utilities|Enum|Equal(Enum) (Variables|UserOptions|GetPlay%20Audio%20/%20VFX) "NewEnumerator1") (Rendering|IsVisible (Variables|Debug|GetStaticMeshComponent)))))

(fn TeamSettingUpgradePath
  (bind _toyoptionscomponent (Variables|Default|GetToyOptionsComponent))
  (bind (_currentoptions _returnvalue) (OptionData|GetPropertyOverrides _toyoptionscomponent))
  (bind (_value _returnvalue_2) (Utilities|Map|Find _currentoptions "NewTeamIndex"))
  (bind _toyoptionscomponent_2 (Variables|Default|GetToyOptionsComponent))
  (Networking|SwitchHasAuthority
    (:Authority
      (if _returnvalue
        (if _returnvalue_2
          (bind _output_get (LocalVariables|SetNewTeamIndexPropertyOverride _value))
          (bind _output_get_2 (LocalVariables|SetActualTeamIndexOverride (Utilities|String|StringToInteger _output_get)))
          (OptionData|SetPropertyOverrides _toyoptionscomponent_2 (Utilities|Map|MakeMap (Class|PlaylistUserOptionBase|GetOptionKey (Variables|Default|GetTeamSettingOption) false 0) (Class|PlaylistUserOptionBase|GetOptionValueNamefromValue (Variables|Default|GetTeamSettingOption) (Utilities|String|ToString(Integer) (Utilities|Select :Index _output_get_2 :Option%200 15 :Option%201 -1 :Option%202 -1 :Option%203 0 :Option%204 1 :Option%205 2 :Option%206 3 :Option%207 4 :Option%208 5 :Option%209 6 :Option%2010 7 :Option%2011 8 :Option%2012 9 :Option%2013 10 :Option%2014 11 :Option%2015 12 :Option%2016 13 :Option%2017 14)) true 0)))
          (OptionData|RemovePropertyOverrides _toyoptionscomponent_2 (Utilities|Array|MakeArray "NewTeamIndex")))))))

(fn UnbindAllCurrentMinigamePlayers
  (bind _fortminigameobjective (Variables|Default|GetFortMinigameObjective))
  (bind _currentminigame (Variables|/Script/FortniteGame.FortMinigameLogicComponent|GetCurrentMinigame _fortminigameobjective))
  (Utilities|IsValid _currentminigame
    (:Is%20Valid
      (bind _output_get (LocalVariables|SetParticipatingPlayers (Minigame|GetParticipatingPlayers _currentminigame)))
      (for _array_element _output_get
        (bind _asfort_player_controller_athena (Utilities|Casting|CastToFortPlayerControllerAthena (Fortnite|GetFortPlayerControllerfromActor _array_element)))
        (bind _output_get_2 (LocalVariables|SetCurrentParticipatingPlayer _asfort_player_controller_athena))
        (if (Pawn|IsLocalPlayerController _output_get_2)
          (FortPlayerControllerAthena|UnbindEventfromOnLocalPlayerChangedTeams _output_get_2 (EventDispatchers|CreateEvent :!event UpdateFromTeamChanged)))))))

(fn UnbindPlayer
  (params (UniqueNetID :type Unique%20Net%20Id%20Structure%20%28by%20ref%29))
  (bind _output_get (LocalVariables|SetLocalPlayerControllers (Fortnite|GetLocalFortPlayerControllers true)))
  (bind _array_element (Utilities|Array|ForEachLoopwithBreak _output_get)
    (:LoopBody
      (bind _asfort_player_controller_athena (Utilities|Casting|CastToFortPlayerControllerAthena _array_element))
      (bind _output_get_2 (LocalVariables|SetCurrentLocalPlayer _asfort_player_controller_athena))
      (if (Online|UniqueNetIdRepl|Equal(UniqueIdRepl) (Default|GetGameAccountId _output_get_2) (Variables|GetUniqueNetID))
        (FortPlayerControllerAthena|UnbindEventfromOnLocalPlayerChangedTeams (LocalVariables|GetCurrentLocalPlayer) (EventDispatchers|CreateEvent :!event UpdateFromTeamChanged))))))

(fn UpdateBeaconLocation
  (bind (_origin _boxextent) (Collision|GetComponentBounds (Variables|Debug|GetStaticMeshComponent)))
  (Transformation|SetWorldLocationAndRotation (Variables|Default|GetParticleSystem) (Math|Vector|vector+vector _origin (Math|Vector|MakeVector 0.0 0.0 (.z _boxextent))) "0, 0, 0" false false))

(fn UpdateComponentVisibility
  (Visibility|SetVisiblityDuringPhase (Variables|Default|GetVisibleInGameComponent) (select (Variables|UserOptions|Basic|GetVisible%20during%20games) "NewEnumerator3" "NewEnumerator1")))

(fn UpdateFromTeamChanged
  (CallFunction|SetParticle)
  (Variables|Default|SetLocalPlayerFriendly (Team|OnSameTeam (Fortnite|GetFirstLocalFortPlayerController true) self))
  (Default|CallLocalPlayerChangedTeams))

(fn UpdateHealthEffects
  (params (NewHealth :type Float%20%28double-precision%29))
  (bind _output_get (LocalVariables|SetHealthPercent (/ NewHealth (Variables|UserOptions|Health|GetHealth))))
  (if (> _output_get 0.5)
    (if (Variables|Default|GetbDamageEffectsAreActive)
      (CallFunction|StopHealthEffects)
      (Variables|Default|SetbDamageEffectsAreActive false))
    (elif (> (LocalVariables|GetHealthPercent) 0.25)
      (CallFunction|Play50%25Effect)
      (Variables|Default|SetbDamageEffectsAreActive true)
      (elif (> (LocalVariables|GetHealthPercent) 0.0)
        (CallFunction|Play25%25Effect)
        (Variables|Default|SetbDamageEffectsAreActive true)
        (elif (Variables|Default|GetbDamageEffectsAreActive)
          (CallFunction|StopHealthEffects)
          (Variables|Default|SetbDamageEffectsAreActive false))))))

(fn UpdateStaticMeshCollision
  (bind _collisionprofilename (LocalVariables|GetCollisionProfileName))
  (Utilities|FlowControl|Sequence
    (:then_0
      (LocalVariables|SetNoCollisionProfileName (CreativeDevices|GetNamefromCollisionProfile "(Name=\"FortCreativeOnlyVolumeManager\")"))
      (LocalVariables|SetEditModeCollisionProfileName (CreativeDevices|GetNamefromCollisionProfile "(Name=\"FortDynamicMeshPhysics\")"))
      (LocalVariables|SetPlayModeProfileName (Utilities|Map|Find (Variables|Default|GetCollides%20With%20Names) (Variables|UserOptions|GetCollides%20With)))
      (if (Default|IsinPlayMode)
        (switch Utilities|FlowControl|Switch|SwitchonEnum_ForcedValueDuringGame (Variables|Replicated|GetCollisionForcedDuringGame)
          (:NewEnumerator0
            (switch Utilities|FlowControl|Switch|SwitchonEnum_CollisionDuringGamesSetting (Variables|UserOptions|GetCollision%20During%20Games)
              (:NewEnumerator0
                (LocalVariables|SetCollisionProfileName (LocalVariables|GetNoCollisionProfileName)))
              (:NewEnumerator1
                (LocalVariables|SetCollisionProfileName (LocalVariables|GetPlayModeProfileName)))
              (:NewEnumerator3
                (if (Variables|UserOptions|Basic|GetVisible%20during%20games)
                  (LocalVariables|SetCollisionProfileName (LocalVariables|GetPlayModeProfileName))
                  (else
                    (LocalVariables|SetCollisionProfileName (LocalVariables|GetNoCollisionProfileName)))))))
          (:NewEnumerator1
            (LocalVariables|SetCollisionProfileName (LocalVariables|GetPlayModeProfileName)))
          (:NewEnumerator2
            (LocalVariables|SetCollisionProfileName (LocalVariables|GetNoCollisionProfileName))))
        (else
          (LocalVariables|SetCollisionProfileName (LocalVariables|GetEditModeCollisionProfileName)))))
    (:then_1
      (if (Utilities|Name|NotEqual(Name) _collisionprofilename "None")
        (Collision|SetCollisionProfileName (Variables|Debug|GetStaticMeshComponent) _collisionprofilename true :!class /Script/Engine.PrimitiveComponent)))))

(fn UpdateTrackedProgress
  (bind _returnvalue (Health|GetHealthPercent))
  (Networking|SwitchHasAuthority
    (:Authority
      (CallFunction|HandleHealthChangeEvents _returnvalue)
      (Minigame|Objective|SetTrackedProgress (Variables|Default|GetFortMinigameObjective) _returnvalue))))

(fn UpdateTrackingState
  (params (Minigame :type Fort%20Minigame%20Object%20Reference))
  (Utilities|IsValid Minigame
    (:Is%20Valid
      (if (Variables|UserOptions|UI|GetShow%20In%20Objective%20HUD)
        (Minigame|HasStarted (Variables|GetMinigame) :!class /Script/FortniteGame.FortMinigame
          (:Yes
            (if (not (Minigame|IsPostGame (Variables|GetMinigame)))
              (Minigame|Objective|StartTracking (Variables|Default|GetFortMinigameObjective) (- (Utilities|Array|FindItem (Variables|Default|GetTextureArrayForIndicator) (Utilities|Struct|BreakCreativeIconOption (Variables|UserOptions|UI|GetObjective%20Identifier))) 1) (Health|GetHealthPercent) "DestructionObjective" 0 0 true)))
          (:No
            _))
        (Minigame|Objective|StopTracking (Variables|Default|GetFortMinigameObjective) true (Health|GetHealthPercent))
        (else
          _)))))

(fn ConstructionScript
  (|CallParentFunction)
  (CallFunction|InitializeDevice))
