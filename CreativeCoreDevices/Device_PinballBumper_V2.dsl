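; Creative pinball bumper device that knocks players, vehicles and physics objects sideways or upward, awards score to players and resets after a configurable time.
; Trigger overlaps send gameplay events to an equipped ability set; launch events blend push-from-center with reflected velocity via ObjectDirectionImportance.
; Physics props bounce through a FortPhysicsImpulse component fed the device options.
; Gameplay cues drive timelines, effects and cached glow materials; legacy indexed options migrate to numeric values.
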(event OnComponentBeginOverlap(UpTrigger)
  (params (OverlappedComponent :type Primitive%20Component%20Object%20Reference) (OtherActor :type Actor%20Object%20Reference) (OtherComp :type Primitive%20Component%20Object%20Reference) (OtherBodyIndex :type Integer) (bFromSweep :type Boolean) (SweepResult :type Hit%20Result%20Structure%20%28by%20ref%29))
  ;[1 1 1] Bumper is Triggered
  (if (not (Variables|Default|GetbIsDisabled))
    (Variables|Default|SetbIsTriggeredLaunch false)
    (CallFunction|AttemptKnock OtherActor true)))

(event Custom|Enable%20Event
  (params (Instigator :type Controller%20Object%20Reference))
  ;[0.6337 0 1] Gameplay Functions
  (CallFunction|ResetBumper)
  (CallFunction|UpdatePhysicsThreadSettings))

(event OnComponentBeginOverlap(Trigger)
  (params (OverlappedComponent :type Primitive%20Component%20Object%20Reference) (OtherActor :type Actor%20Object%20Reference) (OtherComp :type Primitive%20Component%20Object%20Reference) (OtherBodyIndex :type Integer) (bFromSweep :type Boolean) (SweepResult :type Hit%20Result%20Structure%20%28by%20ref%29))
  ;[1 1 1] Bumper is Triggered
  (if (not (Variables|Default|GetbIsDisabled))
    (Variables|Default|SetbIsTriggeredLaunch false)
    (CallFunction|AttemptKnock OtherActor)))

(event Custom|Activate%20Event
  (params (Instigator :type Controller%20Object%20Reference))
  ;[0.6337 0 1] Gameplay Functions
  (Variables|Default|SetbIsTriggeredLaunch true)
  (CallFunction|AttemptKnock 0 true))

(event Custom|Disable%20FX
  ;[1 1 1] Resetting the Top Mesh Element
  (if (not (Editor|InEditor))
    (bind (_direction _pinballbumper_03_z_offset) (Timeline|LowerTop PlayFromStart :NewTime (Components|Timeline|GetTimelineLength (Variables|Device_PinballBumper_V2|GetRaiseTop)))
      (:Update
        (Transformation|SetRelativeLocation (Variables|Default|GetOC_PinballBumper_03) (Math|Vector|MakeVector 0.0 0.0 _pinballbumper_03_z_offset) false false)))
    (else
      (Timeline|LowerTop SetNewTime))))

(event Custom|Disable%20Event
  (params (Instigator :type Controller%20Object%20Reference))
  ;[0.6337 0 1] Gameplay Functions
  (CallFunction|DisableBumper true)
  (CallFunction|UpdatePhysicsThreadSettings))

(event Custom|Reset%20FX
  ;[1 1 1] Resetting the Top Mesh Element
  (if (not (Editor|InEditor))
    (Audio|PlaySoundatLocation (Variables|Sound|GetBumper%20Ready%20Sound) (Transformation|GetActorLocation))
    (bind (_direction _newtrack_0) (Timeline|RaiseTop PlayFromStart :NewTime (Components|Timeline|GetTimelineLength (Variables|Device_PinballBumper_V2|GetRaiseTop)))
      (:Update
        (Transformation|SetRelativeLocation (Variables|Default|GetOC_PinballBumper_03) (Math|Vector|MakeVector 0.0 0.0 _newtrack_0) false false)))
    (else
      (Timeline|RaiseTop SetNewTime))))

(event OnPhysicsImpulseOverlapTriggeredSignatureDelegate(FortPhysicsImpulse)
  (params (Payload :type Fort%20Physics%20Impulse%20Component%20Delegate%20Payload%20Structure))
  (bind (_fortphysicsimpulsecomponent _triggeredobjects_payload) (Utilities|Struct|BreakFortPhysicsImpulseComponentDelegatePayload Payload))
  ;[1 1 1] Deal damage, trigger cooldown, and Play VFX/SFX when a prop is triggered via the FortPhysicsImpulseComponent. Limit how often FX are triggered.
  (for _array_element _triggeredobjects_payload
    (bind _asactor (Utilities|Casting|CastToActor _array_element))
    (CallFunction|OnFinishedAction)
    (if (or (> (- (Utilities|Time|GetTimeSeconds) (Utilities|Map|Find (Variables|Default|GetLastFXPlayTimesPerObject) _asactor)) 1.0) (not (.ReturnValue (Utilities|Map|Find (Variables|Default|GetLastFXPlayTimesPerObject) _asactor))))
      (Utilities|FlowControl|Sequence
        (:then_0
          (Networking|CanExecuteCosmeticEvents
            (:True
              (Audio|SpawnSoundAttached (Variables|Sound|GetPushback%20Sound) (Variables|/Script/Engine.Actor|GetRootComponent _asactor)))))
        (:then_1
          (GameplayCue|AddGameplayCue self (Utilities|Struct|BreakGameplayCueTag (Variables|Default|GetKnockCueTag)))
          (CallFunction|SendActivationEvent _asactor))
        (:then_2
          (Utilities|Map|Add (Variables|Default|GetLastFXPlayTimesPerObject) _asactor (Utilities|Time|GetTimeSeconds)))))))

(event EventBeginPlay
  ;[1 1 1] Initialization
  (|CallParentFunction)
  (Networking|SwitchHasAuthority
    (:Authority
      (Ability|EquipFortAbilitySet self "/CreativeCoreDevices/PinballBumper/GAS_Creative_PinballBumper_V2.GAS_Creative_PinballBumper_V2"))
    (:Remote
      _))
  (if (Fortnite|IsPhysicallySimulatedIsland)
    (CallFunction|UpdatePhysicsThreadSettings)))

(event Custom|LaunchPhysicsObject
  (params (Target :type Fort%20Physics%20Pawn%20Object%20Object%20Reference) (ImpactPoint :type Vector) (TargetVelocity :type Vector))
  ;[1 0.7546 0] Launch Objects
  (Utilities|IsValid Target
    (:Is%20Valid
      (bind _vector (CallFunction|GenericKnockbackActions Target ImpactPoint TargetVelocity))
      (PhysicsObject|ApplyImpulse Target self _vector))))

(event Custom|LaunchPhysicsObjectUp
  (params (PhysicsObject :type Fort%20Physics%20Pawn%20Object%20Object%20Reference) (ObjectVelocity :type Vector))
  ;[1 0.7546 0] Launch Objects
  (Utilities|IsValid PhysicsObject
    (:Is%20Valid
      (Physics|AddImpulse (Variables|/Script/FortniteGame.FortPhysicsPawnObject|GetPhysicsMeshComponent PhysicsObject) (CallFunction|GetKnockbackUpVector ObjectVelocity PhysicsObject) "None" true))))

(event MinigameStateChangedDelegate(FortMinigameProgress)
  (params (Minigame :type Fort%20Minigame%20Object%20Reference) (NewMinigameState :type EFortMinigameState%20Enum :default "PreGame"))
  ;[1 1 1] Initialization
  (if (Utilities|Enum|Equal(Enum) NewMinigameState "PreGame")
    (CallFunction|BumperEnableOnStart)))

(event Custom|LaunchVehicle
  (params (Target :type Fort%20Athena%20Vehicle%20Object%20Reference) (ImpactPoint :type Vector) (TargetVelocity :type Vector))
  ;[1 0.7546 0] Launch Objects
  (Utilities|IsValid Target
    (:Is%20Valid
      (Score|AwardScoreVehiclePassengers Target)
      (bind _vector (CallFunction|GenericKnockbackActions Target ImpactPoint TargetVelocity))
      (Audio|BroadcastAbilityImpactatLocation 0 Target _vector 777.0))))

(event Custom|LaunchVehicleUp
  (params (Target :type Fort%20Athena%20Vehicle%20Object%20Reference) (TargetVelocity :type Vector))
  ;[1 0.7546 0] Launch Objects
  (Utilities|IsValid Target
    (:Is%20Valid
      (Score|AwardScoreVehiclePassengers Target)
      (Audio|BroadcastAbilityImpactatLocation 0 Target (CallFunction|GetKnockbackUpVector TargetVelocity Target) 777.0))))

(event OnPropertiesLoaded(ToyOptionsComponent)
  (bind _bumpercolor (Variables|UserOptions|Basic|GetBumperColor))
  (bind _bumpermaterials (Variables|Default|GetBumperMaterials))
  (bind (_value _returnvalue) (Utilities|Map|Find _bumpermaterials _bumpercolor))
  (bind _output (Utilities|ResolveSoftReference _value))
  ;[1 1 1] Initialization
  (Class|CreativeUIBroadcastComponent|Initialize (Variables|Default|GetCreative_HUDBroadcast_Component) (Variables|Default|GetFortMinigameProgress) (Variables|UserOptions|Advanced|GetDisplayScoreUpdateOnHUD))
  (Utilities|FlowControl|Sequence
    (:then_0
      (if _returnvalue
        (if (Utilities|IsValidSoftObjectReference _value)
          (Utilities|IsValid _output
            (:Is%20Valid
              (CallFunction|UpdateParentMaterial _output))
            (:Is%20Not%20Valid
              (bind _loaded_object (Creative|Loading|AsyncLoadEditorSafe _value)
                (:Loaded
                  (bind _asmaterial_interface (Utilities|Casting|CastToMaterialInterface _loaded_object))
                  (CallFunction|UpdateParentMaterial _asmaterial_interface))))))))
    (:then_1
      (CallFunction|BumperEnableOnStart))
    (:then_2
      (CallFunction|UpdatePhysicsThreadSettings))))

(event OnAnyPropertyChanged(ToyOptionsComponent)
  (bind _bumpercolor (Variables|UserOptions|Basic|GetBumperColor))
  (bind _bumpermaterials (Variables|Default|GetBumperMaterials))
  (bind (_value _returnvalue) (Utilities|Map|Find _bumpermaterials _bumpercolor))
  (bind _output (Utilities|ResolveSoftReference _value))
  ;[1 1 1] Initialization
  (Class|CreativeUIBroadcastComponent|Initialize (Variables|Default|GetCreative_HUDBroadcast_Component) (Variables|Default|GetFortMinigameProgress) (Variables|UserOptions|Advanced|GetDisplayScoreUpdateOnHUD))
  (Utilities|FlowControl|Sequence
    (:then_0
      (if _returnvalue
        (if (Utilities|IsValidSoftObjectReference _value)
          (Utilities|IsValid _output
            (:Is%20Valid
              (CallFunction|UpdateParentMaterial _output))
            (:Is%20Not%20Valid
              (bind _loaded_object (Creative|Loading|AsyncLoadEditorSafe _value)
                (:Loaded
                  (bind _asmaterial_interface (Utilities|Casting|CastToMaterialInterface _loaded_object))
                  (CallFunction|UpdateParentMaterial _asmaterial_interface))))))))
    (:then_1
      (CallFunction|BumperEnableOnStart))
    (:then_2
      (CallFunction|UpdatePhysicsThreadSettings))))

(event Custom|LaunchPawn
  (params (Pawn :type Fort%20Pawn%20Object%20Reference) (ImpactPoint :type Vector) (PawnVelocity :type Vector))
  ;[1 0.7546 0] Launch Objects
  (Utilities|IsValid Pawn
    (:Is%20Valid
      (Score|AwardScore Pawn)
      (bind _vector (CallFunction|GenericKnockbackActions Pawn ImpactPoint PawnVelocity))
      (Pawn|Character|LaunchCharacterJump Pawn _vector true true (not (Variables|UserOptions|Basic|GetbFallDamage)) false))))

(event EditorPreview|EventSetupForEditorPreview
  (bind _bumpercolor (Variables|UserOptions|Basic|GetBumperColor))
  (bind _bumpermaterials (Variables|Default|GetBumperMaterials))
  (bind (_value _returnvalue) (Utilities|Map|Find _bumpermaterials _bumpercolor))
  (bind _output (Utilities|ResolveSoftReference _value))
  ;[1 1 1] Initialization
  (Utilities|FlowControl|Sequence
    (:then_0
      (if _returnvalue
        (if (Utilities|IsValidSoftObjectReference _value)
          (Utilities|IsValid _output
            (:Is%20Valid
              (CallFunction|UpdateParentMaterial _output))
            (:Is%20Not%20Valid
              (bind _loaded_object (Creative|Loading|AsyncLoadEditorSafe _value)
                (:Loaded
                  (bind _asmaterial_interface (Utilities|Casting|CastToMaterialInterface _loaded_object))
                  (CallFunction|UpdateParentMaterial _asmaterial_interface))))))))
    (:then_1
      (CallFunction|BumperEnableOnStart))
    (:then_2
      (CallFunction|UpdatePhysicsThreadSettings))))

(event Custom|LaunchPawnUp
  (params (Pawn :type Fort%20Pawn%20Object%20Reference) (PawnVelocity :type Vector))
  ;[1 0.7546 0] Launch Objects
  (Utilities|IsValid Pawn
    (:Is%20Valid
      (Score|AwardScore Pawn)
      (Pawn|Character|LaunchCharacterJump Pawn (CallFunction|GetKnockbackUpVector PawnVelocity Pawn) true true (not (Variables|UserOptions|Basic|GetbFallDamage)) false))))

(event GameplayCue|GameplayCue.Abilities.Activation.Devices.PinballBumper.KnockSide
  (params (EventType :type EGameplayCueEvent%20Enum :default "OnActive") (Parameters :type Gameplay%20Cue%20Parameters%20Structure%20%28by%20ref%29))
  ;[1 1 1] Cosmetic Effects on Hit
  (Networking|CanExecuteCosmeticEvents
    (:True
      (CallFunction|SwitchToDyamicMaterial)
      (bind _material_instance_dynamic (Variables|Default|GetMaterial%20Instance%20Dynamic)
        (:then
          (Rendering|Material|SetScalarParameterValue _material_instance_dynamic "GlowOn" 1.0 :!class /Script/Engine.MaterialInstanceDynamic))
        (:else
          _))
      (bind _returnvalue (Effects|SpawnEmitteratLocation "/CreativeCoreDevices/Athena/Effects/Particles/Creative/Props/P_Creative_Pinball_Bumper.P_Creative_Pinball_Bumper" (Transformation|GetWorldLocation (Variables|Default|GetBody))))
      (Effects|Components|ParticleSystem|SetColorParameter _returnvalue "LightColor" (Utilities|Map|Find (Variables|Default|GetFX_Colors) (Variables|UserOptions|Basic|GetBumperColor)))
      (Audio|PlaySoundatLocation (Variables|Sound|GetBounce%20Sound) (Transformation|GetActorLocation))
      (bind (_direction _pinballbumper_02_scale) (Timeline|Bounce PlayFromStart)
        (:Update
          (Transformation|SetRelativeScale3D (Variables|Default|GetOC_PinballBumper_02) (Math|Vector|MakeVector _pinballbumper_02_scale _pinballbumper_02_scale 1.0)))
        (:Finished
          (bind _material_instance_dynamic_2 (Variables|Default|GetMaterial%20Instance%20Dynamic)
            (:then
              (Rendering|Material|SetScalarParameterValue _material_instance_dynamic_2 "GlowOn" 0.0 :!class /Script/Engine.MaterialInstanceDynamic)
              (CallFunction|SwitchToConstantMaterial))
            (:else
              _))
          (if (not (> (Variables|UserOptions|Basic|GetResetTime) 0.0))
            (CallFunction|ResetFX)))))))

(event GameplayCue|GameplayCue.Abilities.Activation.Devices.PinballBumper.KnockUp
  (params (EventType :type EGameplayCueEvent%20Enum :default "OnActive") (Parameters :type Gameplay%20Cue%20Parameters%20Structure%20%28by%20ref%29))
  ;[1 1 1] Cosmetic Effects on Hit
  (Networking|CanExecuteCosmeticEvents
    (:True
      (Audio|PlaySoundatLocation (Variables|Sound|GetBounce%20Sound) (Transformation|GetActorLocation))
      (CallFunction|SwitchToDyamicMaterial)
      (bind _material_instance_dynamic (Variables|Default|GetMaterial%20Instance%20Dynamic)
        (:then
          (Rendering|Material|SetScalarParameterValue _material_instance_dynamic "GlowOn" 1.0 :!class /Script/Engine.MaterialInstanceDynamic)
          (bind (_direction _pinballbumper_03_z_offset) (Timeline|BounceUp PlayFromStart)
            (:Update
              (Transformation|SetRelativeLocation (Variables|Default|GetOC_PinballBumper_03) (Math|Vector|MakeVector 0.0 0.0 _pinballbumper_03_z_offset) false false))
            (:Finished
              (bind _material_instance_dynamic_2 (Variables|Default|GetMaterial%20Instance%20Dynamic)
                (:then
                  (Rendering|Material|SetScalarParameterValue _material_instance_dynamic_2 "GlowOn" 0.0 :!class /Script/Engine.MaterialInstanceDynamic)
                  (CallFunction|SwitchToConstantMaterial))
                (:else
                  _))
              (if (not (> (Variables|UserOptions|Basic|GetResetTime) 0.0))
                (CallFunction|ResetFX)))))
        (:else
          (Timeline|BounceUp PlayFromStart))))))

(event Custom|Switch%20to%20Dyamic%20Material
  ;[1 1 1] Material Changes
  (Networking|CanExecuteCosmeticEvents
    (:True
      (Variables|Default|GetMaterial%20Instance%20Dynamic
        (:then
          _)
        (:else
          (bind _material_cache (Variables|Default|GetMaterial%20Cache)
            (:then
              (bind (_outmaterialinstance _returnvalue) (Class|FortMaterialCache|ReserveMaterial _material_cache))
              (Variables|Default|SetMaterial%20Instance%20Dynamic _outmaterialinstance)
              (if _returnvalue
                _))
            (:else))))
      (Rendering|Material|SetMaterial (Variables|Default|GetOC_PinballBumper_03) 0 (Variables|Default|GetMaterial%20Instance%20Dynamic)))))

(event Custom|Switch%20to%20Constant%20Material
  ;[1 1 1] Material Changes
  (Networking|CanExecuteCosmeticEvents
    (:True
      (Rendering|Material|SetMaterial (Variables|Default|GetOC_PinballBumper_03) 0 (Variables|Default|GetParentMaterial))
      (bind _material_cache (Variables|Default|GetMaterial%20Cache)
        (:then
          (Class|FortMaterialCache|ReleaseMaterial _material_cache (Variables|Default|GetMaterial%20Instance%20Dynamic)))
        (:else)))))

(fn AttemptKnock
  (params (Actor :type Actor%20Object%20Reference) (Up :type Boolean))
  (Networking|SwitchHasAuthority
    (:Authority
      (Utilities|Casting|CastToBuildingProp Actor
        (:CastFailed
          (Utilities|CanTrigger Actor
            (:True
              (Ability|SendGameplayEventtoActor self (select Up (Variables|Default|GetKnockUpCue) (Variables|Default|GetKnockSideCue)))
              (CallFunction|SendActivationEvent Actor))))))))

(fn AwardScore
  (params (Pawn :type Pawn%20Object%20Reference))
  (Networking|SwitchHasAuthority
    (:Authority
      (bind _output_get (LocalVariables|SetParticipantController (CreativeAI|Controller|GetParticipantControllerfromActor Pawn false)))
      (Utilities|IsValid _output_get
        (:Is%20Valid
          (bind _returnvalue (CreativeDevices|GetMinigameStatsComp (LocalVariables|GetParticipantController) "Athena"))
          (Utilities|IsValid _returnvalue
            (:Is%20Valid
              (Class|FortMinigameStatsComponent|SendStatEventwithTags _returnvalue (LocalVariables|GetParticipantController) "MinigameDynamicEvent" self (GameplayTags|MakeGameplayTagContainerfromTag "(TagName=\"Minigame.Stat.Score.Pinball\")") (Class|FortMinigameStatsComponent|GetSourceandContextTags _returnvalue (LocalVariables|GetParticipantController)) (.OutContextTags (Class|FortMinigameStatsComponent|GetSourceandContextTags _returnvalue (LocalVariables|GetParticipantController))) (Variables|UserOptions|Basic|GetScoreValue))
              (if (Variables|UserOptions|Advanced|GetDisplayScoreUpdateOnHUD)
                (Class|CreativeHUDBroadcastComponent|DoScoreEventforParticipant (Variables|Default|GetCreative_HUDBroadcast_Component) (LocalVariables|GetParticipantController) (Variables|UserOptions|Basic|GetScoreValue) true)))))))))

(fn AwardScoreVehiclePassengers
  (params (Vehicle :type Fort%20Athena%20Vehicle%20Object%20Reference))
  (Networking|SwitchHasAuthority
    (:Authority
      (for _array_element (Vehicle|GetAllPassengers Vehicle)
        (Utilities|IsValid _array_element
          (:Is%20Valid
            (Score|AwardScore _array_element)))))))

(fn BlueprintCanInteract
  (params (InteractingPawn :type Fort%20Pawn%20Object%20Reference) (InteractionBeingAttempted :type EInteractionBeingAttempted%20Enum :default "FirstInteraction") (InteractionType :type TInteractionType%20Enum :default "IT_NoInteraction"))
  (bind _returnvalue (Pawn|GetController InteractingPawn))
  (bind _self self)
  (bind _returnvalue_2 (Minigame|GetMinigame _self :!class /Script/FortniteGame.MinigameBlueprintLibrary))
  (Utilities|FlowControl|Sequence
    (:then_0
      (Utilities|IsValid (Variables|/Script/FortniteGame.FortActorOptionsComponent|GetInternal_PlayerOptions (Variables|Default|GetToyOptionsComponent))
        (:Is%20Valid
          (Utilities|IsValid InteractingPawn
            (:Is%20Valid
              (Utilities|IsValid _returnvalue
                (:Is%20Valid
                  (bind _asfort_player_controller_athena (Utilities|Casting|CastToFortPlayerControllerAthena _returnvalue))
                  (if (Creative|CanPlayerCreateinCurrentExperience _asfort_player_controller_athena)
                    (Utilities|IsValid _returnvalue_2
                      (:Is%20Valid
                        (LocalVariables|SetbInteractSuccess (not (Minigame|HasStarted _returnvalue_2 :!class /Script/FortniteGame.FortMinigame))))
                      (:Is%20Not%20Valid
                        (LocalVariables|SetbInteractSuccess true)))))))))))
    (:then_1
      (return (LocalVariables|GetbInteractSuccess)))))

(fn BlueprintGetInteractionString
  (params (InteractingPawn :type Fort%20Pawn%20Object%20Reference) (InteractionBeingAttempted :type EInteractionBeingAttempted%20Enum :default "FirstInteraction"))
  (return "NSLOCTEXT(\"[7D57280D5D88DA7A4FBDB05FB42DB405]\", \"0B82013743F64CAECE8C018E449E0AE1\", \"Customize\")"))

(fn BlueprintOnLocalInteract
  (params (InteractingPawn :type Fort%20Player%20Pawn%20Object%20Reference) (InteractionBeingAttempted :type EInteractionBeingAttempted%20Enum :default "FirstInteraction"))
  (return false))

(fn BumperEnableonStart
  (Networking|SwitchHasAuthority
    (:Authority
      (if (Variables|UserOptions|GetbEnabledAtGameStart)
        (CallFunction|ResetBumper)
        (else
          (CallFunction|DisableBumper true))))))

(fn CalculateLaunchVector
  (params (InputPin :type Actor%20Object%20Reference) (ImpactPoint :type Vector) (KnockbackValue :type Float%20%28double-precision%29) (KnockupValue :type Float%20%28double-precision%29) (InputVelocity :type Vector))
  (bind _asfort_player_pawn (Utilities|Casting|CastToFortPlayerPawn (Variables|GetInputPin))
    (:then
      _)
    (:CastFailed
      _))
  (return (Math|Vector|SelectVector (/ (* (+ (* (Transformation|GetUpVector (Variables|Default|GetBody)) KnockupValue) (* (Math|Vector|Normalize (Math|Vector|ProjectVectorontoPlane (Math|Vector|Lerp(Vector) (* (Math|Vector|Normalize (- (Transformation|GetWorldLocation (Variables|Default|GetBody)) ImpactPoint)) -1.0) (Math|Vector|GetReflectionVector (Math|Vector|Normalize (Math|Vector|SelectVector InputVelocity (Math|Vector|Normalize (Math|Vector|MakeVector (.x InputVelocity) (.y InputVelocity))) (Fortnite|IsPhysicallySimulatedIsland))) (Math|Vector|Normalize (- (Transformation|GetWorldLocation (Variables|Default|GetBody)) ImpactPoint))) (/ (Variables|UserOptions|GetObjectDirectionImportance) 1e+02)) (Transformation|GetUpVector (Variables|Default|GetTrigger)))) KnockbackValue)) (Variables|Default|GetLaunchForceConversionFactor)) (Variables|Default|GetPlayerMass)) (+ (* (Transformation|GetUpVector (Variables|Default|GetBody)) KnockupValue) (* (Math|Vector|Normalize (Math|Vector|ProjectVectorontoPlane (Math|Vector|Lerp(Vector) (* (Math|Vector|Normalize (- (Transformation|GetWorldLocation (Variables|Default|GetBody)) ImpactPoint)) -1.0) (Math|Vector|GetReflectionVector (Math|Vector|Normalize (Math|Vector|SelectVector InputVelocity (Math|Vector|Normalize (Math|Vector|MakeVector (.x InputVelocity) (.y InputVelocity))) (Fortnite|IsPhysicallySimulatedIsland))) (Math|Vector|Normalize (- (Transformation|GetWorldLocation (Variables|Default|GetBody)) ImpactPoint))) (/ (Variables|UserOptions|GetObjectDirectionImportance) 1e+02)) (Transformation|GetUpVector (Variables|Default|GetTrigger)))) KnockbackValue)) (and (Utilities|IsValid _asfort_player_pawn :!class /Script/Engine.KismetSystemLibrary) (Utilities|Enum|Equal(Enum) (Variables|UserOptions|GetImpulseOrVelocity) "Force")))))

(fn DisableBumper
  (params (DisableCollision :type Boolean))
  (Variables|Default|SetbIsDisabled true)
  (if DisableCollision
    (Collision|SetCollisionEnabled (Variables|Default|GetTrigger) "NoCollision")))

(fn GenericKnockbackActions
  (params (Pawn :type Pawn%20Object%20Reference) (ImpactPoint :type Vector) (PawnVelocity :type Vector))
  (bind (_vehicleknockbackvalue _pawnknockbackvalue _vehicleknockupvalue _pawnknockupvalue) (CallFunction|GetKnockbackValue))
  (LocalVariables|Setlocal%20Pawn Pawn)
  (Utilities|FlowControl|Sequence
    (:then_0
      ;[1 1 1] Figure out which knockback values to use
      (Utilities|Casting|CastToFortPlayerPawnAthena (LocalVariables|Getlocal%20Pawn)
        (:then
          (LocalVariables|Setlocal%20Knockback _pawnknockbackvalue)
          (LocalVariables|Setlocal%20Knockup _pawnknockupvalue))
        (:CastFailed
          (LocalVariables|Setlocal%20Knockback _vehicleknockbackvalue)
          (LocalVariables|Setlocal%20Knockup _vehicleknockupvalue))) ; is thje object a player pawn? If so, use player values, \nelse use Vehicle Values
    )
    (:then_1
      (Networking|CanExecuteCosmeticEvents
        (:True
          (Audio|SpawnSoundAttached (Variables|Sound|GetPushback%20Sound) (Variables|/Script/Engine.Actor|GetRootComponent (LocalVariables|Getlocal%20Pawn))))))
    (:then_2
      (return (CallFunction|CalculateLaunchVector (LocalVariables|Getlocal%20Pawn) ImpactPoint (LocalVariables|Getlocal%20Knockback) (LocalVariables|Getlocal%20Knockup) PawnVelocity)))))

(fn GetKnockbackUpVector
  (params (TargetVelocity :type Vector) (Pawn :type Pawn%20Object%20Reference))
  (bind (_vehicleknockbackvalue _pawnknockbackvalue) (CallFunction|GetKnockbackValue))
  (bind _asfort_player_pawn (Utilities|Casting|CastToFortPlayerPawn (Variables|GetPawn))
    (:then
      _)
    (:CastFailed
      _))
  (return (Math|Vector|SelectVector (/ (* (Math|Vector|MakeVector (.x (* (Math|Vector|Normalize TargetVelocity) _pawnknockbackvalue)) (.y (* (Math|Vector|Normalize TargetVelocity) _pawnknockbackvalue)) (.z (* (Transformation|GetUpVector (Variables|Default|GetBody)) _pawnknockbackvalue))) (Variables|Default|GetLaunchForceConversionFactor)) (Variables|Default|GetPlayerMass)) (Math|Vector|MakeVector (.x (* (Math|Vector|Normalize TargetVelocity) _pawnknockbackvalue)) (.y (* (Math|Vector|Normalize TargetVelocity) _pawnknockbackvalue)) (.z (* (Transformation|GetUpVector (Variables|Default|GetBody)) _pawnknockbackvalue))) (and (Utilities|IsValid _asfort_player_pawn :!class /Script/Engine.KismetSystemLibrary) (Utilities|Enum|Equal(Enum) (Variables|UserOptions|GetImpulseOrVelocity) "Force")))))

(fn GetKnockbackValue
  (return (Variables|UserOptions|Basic|GetKnockbackVehicle) (Variables|UserOptions|Basic|GetKnockbackPlayer) (Math|Float|SelectFloat (Variables|UserOptions|GetSideBounceLiftVehicle) (Variables|UserOptions|Basic|GetKnockbackVehicle) (Variables|UserOptions|GetbCustomSideLiftStrength)) (Math|Float|SelectFloat (Variables|UserOptions|GetSideBounceLiftPlayer) (Variables|UserOptions|Basic|GetKnockbackPlayer) (Variables|UserOptions|GetbCustomSideLiftStrength))))

(fn GetOverrideMeshMaterial
  (params (MeshComp :type Static%20Mesh%20Component%20Object%20Reference) (MatIdx :type Integer))
  (if (== MatIdx 0)
    (if (Math|Boolean|ORBoolean (Utilities|Equal(Object) (Variables|Default|GetBody) MeshComp) (Utilities|Equal(Object) (Variables|Default|GetOC_PinballBumper_02) MeshComp) (Utilities|Equal(Object) (Variables|Default|GetOC_PinballBumper_03) MeshComp))
      (bind _material_instance_dynamic (Variables|Default|GetMaterial%20Instance%20Dynamic)
        (:then
          (return _material_instance_dynamic))
        (:else
          (return (Variables|Default|GetParentMaterial)))))
    (else
      (return 0))))

(fn IsTargetableByLinearMover
  (return true))

(fn MessageScore
  (return (Variables|UserOptions|Advanced|GetHUD%20Message) (Variables|UserOptions|Advanced|GetHUD%20MessageScoreColor) (Variables|UserOptions|Advanced|GetHUD%20MessageColor) (Variables|UserOptions|Advanced|GetResetHUD%20MessageScore)))

(fn MigrateLegacyUserOptions
  (bind (_knockbackphysicsprop _sidebounceliftphysicsprop _impulseorvelocity _baffectsplayers _baffectsphysicsprops _knockbackvalueindex) (Class|GetClassDefaults "/CreativeCoreDevices/Device_PinballBumper_V2.Device_PinballBumper_V2_C"))
  (bind (_knockbackphysicsprop_2 _sidebounceliftphysicsprop_2 _impulseorvelocity_2 _baffectsplayers_2 _baffectsphysicsprops_2 _damagelevel) (Class|GetClassDefaults "/CreativeCoreDevices/Device_PinballBumper_V2.Device_PinballBumper_V2_C"))
  (bind (_knockbackphysicsprop_3 _sidebounceliftphysicsprop_3 _impulseorvelocity_3 _baffectsplayers_3 _baffectsphysicsprops_3 _sidebounceliftvalueindex) (Class|GetClassDefaults "/CreativeCoreDevices/Device_PinballBumper_V2.Device_PinballBumper_V2_C"))
  (bind (_knockbackphysicsprop_4 _sidebounceliftphysicsprop_4 _impulseorvelocity_4 _baffectsplayers_4 _baffectsphysicsprops_4 _directionalpha) (Class|GetClassDefaults "/CreativeCoreDevices/Device_PinballBumper_V2.Device_PinballBumper_V2_C"))
  (bind _directionalpha_2 (Variables|Legacy|GetDirectionAlpha))
  (bind _damagelevel_2 (Variables|Legacy|GetDamageLevel))
  (bind _sidebounceliftvalueindex_2 (Variables|Legacy|GetSideBounceLiftValueIndex))
  (bind _knockbackvalueindex_2 (Variables|Legacy|GetKnockbackValueIndex))
  ;[1 1 1] Migration to new UEFN devices
  (|CallParentFunction)
  (Utilities|FlowControl|Sequence
    (:then_0
      (Utilities|FlowControl|Sequence
        (:then_0
          (if (!= _knockbackvalueindex _knockbackvalueindex_2)
            (Variables|UserOptions|Basic|SetKnockbackVehicle (Utilities|Select :Index _knockbackvalueindex_2 :Option%200 0.0 :Option%201 2e+02 :Option%202 5e+02 :Option%203 1e+03 :Option%204 2e+03 :Option%205 5e+03 :Option%206 7e+04))
            (Variables|UserOptions|Basic|SetKnockbackPlayer (Utilities|Select :Index _knockbackvalueindex_2 :Option%200 0.0 :Option%201 6e+02 :Option%202 8e+02 :Option%203 1.6e+03 :Option%204 4e+03 :Option%205 7.5e+03 :Option%206 1e+04))
            ;[1 1 1] Now that we've used the non-default values from the original device, update the value to the defalut and write that to this device so we don't run this migrate logic again
            (Variables|Legacy|SetKnockbackValueIndex _knockbackvalueindex)
            (LocalVariables|SetMigrated true)))
        (:then_1
          ;[1 1 1] SideBounceLiftValueIndex: Was there a customized value on the old device?
          (if (!= _sidebounceliftvalueindex _sidebounceliftvalueindex_2)
            ;[1 1 1] Original Knob allowed, and defaulted to, DoNotOverride which set SideBounceLiftValueIndex=zero which caused SideBounceLift logic to use KnockbackValueIndex for side bouncing
            (Variables|UserOptions|SetbCustomSideLiftStrength true)
            (Variables|UserOptions|SetSideBounceLiftVehicle (Utilities|Select :Index _sidebounceliftvalueindex_2 :Option%200 0.0 :Option%201 0.0 :Option%202 3e+02 :Option%203 6e+02 :Option%204 9e+02 :Option%205 1.2e+03 :Option%206 1.5e+03 :Option%207 1.8e+03))
            (Variables|UserOptions|SetSideBounceLiftPlayer (Utilities|Select :Index _sidebounceliftvalueindex_2 :Option%200 0.0 :Option%201 0.0 :Option%202 4e+02 :Option%203 8e+02 :Option%204 1.2e+03 :Option%205 1.6e+03 :Option%206 2e+03 :Option%207 2.4e+03))
            (Variables|Legacy|SetSideBounceLiftValueIndex _sidebounceliftvalueindex)
            (LocalVariables|SetMigrated true)))
        (:then_2
          (if (!= _directionalpha _directionalpha_2)
            (Variables|UserOptions|SetObjectDirectionImportance (* _directionalpha_2 1e+02))
            (Variables|Legacy|SetDirectionAlpha _directionalpha)
            (LocalVariables|SetMigrated true)))
        (:then_3
          (if (!= _damagelevel _damagelevel_2)
            (Variables|UserOptions|Basic|SetDamage (Utilities|Select :Index _damagelevel_2 :Option%200 0 :Option%201 1 :Option%202 2 :Option%203 5 :Option%204 10 :Option%205 20 :Option%206 50 :Option%207 100 :Option%208 5000))
            (Variables|Legacy|SetDamageLevel _damagelevel)
            (LocalVariables|SetMigrated true)))))
    (:then_1
      ;[1 1 1] Commit the changes so it's saved next time it loads and the logic isn't run again
      (if (LocalVariables|GetMigrated)
        (OptionData|RefreshPropertyOverrides (Variables|Default|GetToyOptionsComponent))))))

(fn OnFinishedAction
  (CallFunction|DisableBumper false)
  (if (Networking|HasAuthority)
    (Utilities|Time|SetTimerbyEvent (EventDispatchers|CreateEvent :!event Reset%20Bumper) (Variables|UserOptions|Basic|GetResetTime))))

(fn OnRep_bIsDisabled
  (if (Variables|Default|GetbIsDisabled)
    (CallFunction|DisableFX)
    (else
      (CallFunction|ResetFX)))
  (CallFunction|UpdatePhysicsThreadSettings))

(fn ResetBumper
  (Variables|Default|SetbIsDisabled false)
  (CallFunction|ResetCollision))

(fn ResetCollision
  (Collision|SetCollisionEnabled (Variables|Default|GetTrigger) (select (Variables|UserOptions|GetbAllowSideBounce) "QueryOnly" "NoCollision"))
  (Collision|SetCollisionEnabled (Variables|Default|GetUpTrigger) (select (Variables|UserOptions|GetbAllowTopBounce) "QueryOnly" "NoCollision")))

(fn SendActivationEvent
  (params (TriggeredActor :type Actor%20Object%20Reference))
  (bind _currentminigame (Variables|/Script/FortniteGame.FortMinigameLogicComponent|GetCurrentMinigame (Variables|Default|GetFortMinigameProgress))
    (:then
      (Minigame|IsInProgress _currentminigame
        (:Yes
          (Utilities|FlowControl|Sequence
            (:then_0
              ;[1 1 1] Find valid controller
              (bind _ascontroller (Utilities|Casting|CastToController (Variables|GetTriggeredActor))
                (:then
                  (LocalVariables|SetInstigatorController _ascontroller))
                (:CastFailed
                  (bind _aspawn (Utilities|Casting|CastToPawn (Variables|GetTriggeredActor)))
                  (Utilities|IsValid (Pawn|GetController _aspawn)
                    (:Is%20Valid
                      (LocalVariables|SetInstigatorController (Pawn|GetController _aspawn)))
                    (:Is%20Not%20Valid
                      ;;[1 1 1] Vehicle
                      (bind _asfort_vehicle_interface (Utilities|Casting|CastToFortVehicleInterface _aspawn))
                      (LocalVariables|SetInstigatorController (Class|FortVehicleInterface|GetVehicleController _asfort_vehicle_interface (Vehicle|GetDriver _asfort_vehicle_interface))))))))
            (:then_1
              ;[1 1 1] Broadcast Event
              (GameplayEvent|BroadcastEvent (Variables|UserOptions|Events|GetOn%20Activated) 0 (Utilities|Struct|MakeGameplayEventActivated (LocalVariables|GetInstigatorController))))))))
    (:else)))

(fn ShouldInvisiblePropsBeAllowed
  (return false))

(fn UpdateParentMaterial
  (params (ParentMaterial :type Material%20Interface%20Object%20Reference))
  (CallFunction|SwitchToConstantMaterial)
  (bind _output_get (Variables|Default|SetParentMaterial ParentMaterial))
  (Utilities|IsValid _output_get
    (:Is%20Valid
      (Rendering|Material|SetMaterial (Variables|Default|GetOC_PinballBumper_02) 0 _output_get)
      (bind _returnvalue (Class|FortMaterialCacheManager|GetCache (Variables|Default|GetMaterial%20Cache%20Settings) (Variables|Default|GetParentMaterial)))
      (Variables|Default|SetMaterial%20Cache _returnvalue))))

(fn UpdatePhysicsThreadSettings
  (Class|FortPhysicsImpulseComponent|SetImpulseDeviceInputPinballBumper (Variables|Default|GetFortPhysicsImpulse) "PinballBumper" (Utilities|Struct|MakeFortPhysicsPinballBumperDeviceInput (Math|Float|SelectFloat (Variables|UserOptions|GetSideBounceLiftPhysicsProp) (Variables|UserOptions|Basic|GetKnockbackPhysicsProp) (Variables|UserOptions|GetbCustomSideLiftStrength)) (Variables|UserOptions|Basic|GetKnockbackPhysicsProp) (Variables|UserOptions|GetObjectDirectionImportance) (Variables|UserOptions|Basic|GetResetTime) (Math|Integer|BitwiseOR 0 (* (Math|Conversions|ToInteger(Boolean) (Variables|UserOptions|GetbAffectsPhysicsProps)) (Math|Conversions|ToInteger(Byte) (Utilities|Enum|LiteralenumEAllowedToBounce "Objects")))) (Variables|UserOptions|GetbAllowSideBounce) (Variables|UserOptions|GetbAllowTopBounce) (not (Variables|Default|GetbIsDisabled)) (Variables|UserOptions|GetImpulseOrVelocity))))

(fn ConstructionScript)
