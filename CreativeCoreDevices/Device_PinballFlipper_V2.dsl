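; Creative pinball flipper device that swings when triggered by proximity, damage or events, launching players, vehicles and physics objects and awarding score.
; Trigger volumes and damage hits resolve swing direction, then fire the flip, activation event and cosmetic glow, sound and particles.
; Knock-up, knockback and bounce-angle options shape launch vectors.
; Physically simulated islands instead use a physics impulse component, launching pawns from hit callbacks to stay on the game thread.
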(event OnAnyPropertyChanged(ToyOptionsComponent)
  ;[0 1 0.1154] Initialization
  (Class|CreativeUIBroadcastComponent|Initialize (Variables|Default|GetCreative_HUDBroadcast_Component) (Variables|Default|GetFortMinigameProgress) (Variables|UserOptions|Advanced|GetDisplayScoreUpdateOnHUD))
  (Utilities|FlowControl|Sequence
    (:then_0
      (CallFunction|LoadMaterials))
    (:then_1
      ;;[1 0.4696 0 0.5] Rework once PhoneTool has callbacks for Preview
      (if (not (Fortnite|IsPhysicallySimulatedIsland))
        (CallFunction|EnableOnStart)))))

(event OnPropertiesLoaded(ToyOptionsComponent)
  ;[0 1 0.1154] Initialization
  (Class|CreativeUIBroadcastComponent|Initialize (Variables|Default|GetCreative_HUDBroadcast_Component) (Variables|Default|GetFortMinigameProgress) (Variables|UserOptions|Advanced|GetDisplayScoreUpdateOnHUD))
  (Utilities|FlowControl|Sequence
    (:then_0
      (CallFunction|LoadMaterials))
    (:then_1
      ;;[1 0.4696 0 0.5] Rework once PhoneTool has callbacks for Preview
      (if (not (Fortnite|IsPhysicallySimulatedIsland))
        (CallFunction|EnableOnStart)))))

(event OnComponentBeginOverlap(TopBox)
  (params (OverlappedComponent :type Primitive%20Component%20Object%20Reference) (OtherActor :type Actor%20Object%20Reference) (OtherComp :type Primitive%20Component%20Object%20Reference) (OtherBodyIndex :type Integer) (bFromSweep :type Boolean) (SweepResult :type Hit%20Result%20Structure%20%28by%20ref%29))
  ;[0 0.5834 1] Trigger the Flipper Events
  (if (Variables|UserOptions|GetbTriggerOnProximity)
    (Variables|Default|SetbIsTriggeredLaunch false)
    (FlipperTriggering|AttemptTriggerFlipper false (Utilities|Select :Index (Variables|UserOptions|Basic|GetFlipDirection) :NewEnumerator0 false :NewEnumerator1 false :NewEnumerator5 true :NewEnumerator2 (Math|Random|RandomBool) :NewEnumerator3 (Math|Random|RandomBool) :NewEnumerator4 (Variables|TriggerDirection|GetFlipFlopDirection)) OtherActor)))

(event OnComponentBeginOverlap(ClockwiseTrigger)
  (params (OverlappedComponent :type Primitive%20Component%20Object%20Reference) (OtherActor :type Actor%20Object%20Reference) (OtherComp :type Primitive%20Component%20Object%20Reference) (OtherBodyIndex :type Integer) (bFromSweep :type Boolean) (SweepResult :type Hit%20Result%20Structure%20%28by%20ref%29))
  ;[0 0.5834 1] Trigger the Flipper Events
  (if (Variables|UserOptions|GetbTriggerOnProximity)
    (Variables|Default|SetbIsTriggeredLaunch false)
    (FlipperTriggering|AttemptTriggerFlipper false true OtherActor)))

(event OnComponentBeginOverlap(CounterClockwiseTrigger)
  (params (OverlappedComponent :type Primitive%20Component%20Object%20Reference) (OtherActor :type Actor%20Object%20Reference) (OtherComp :type Primitive%20Component%20Object%20Reference) (OtherBodyIndex :type Integer) (bFromSweep :type Boolean) (SweepResult :type Hit%20Result%20Structure%20%28by%20ref%29))
  ;[0 0.5834 1] Trigger the Flipper Events
  (if (Variables|UserOptions|GetbTriggerOnProximity)
    (Variables|Default|SetbIsTriggeredLaunch false)
    (FlipperTriggering|AttemptTriggerFlipper false false OtherActor)))

(event MinigameStateChangedDelegate(FortMinigameProgress)
  (params (Minigame :type Fort%20Minigame%20Object%20Reference) (NewMinigameState :type EFortMinigameState%20Enum :default "PreGame"))
  ;[0 1 0.1154] Initialization
  ;;[1 0.8348 0 0.5] Rework once PhoneTool has callbacks for Preview
  (if (not (Fortnite|IsPhysicallySimulatedIsland))
    (if (Utilities|Enum|Equal(Enum) NewMinigameState "PreGame")
      (CallFunction|EnableOnStart))
    (elif (Minigame|IsinPlayMode Minigame :!class /Script/FortniteGame.FortMinigame)
      (if (Utilities|Enum|Equal(Enum) NewMinigameState "Transitioning")
        (CallFunction|SetUpFlipperVectorInitialCached)
        (else
          _))
      (PhysicsSettings|NewtonSetup)
      (CallFunction|EnableOnStart)
      (else
        (PhysicsSettings|NewtonTeardown)
        (FlipperTriggering|DisableFlipper)))))

(event EventOnDamageServer
  (params (Damage :type Float%20%28single-precision%29) (DamageTags :type Gameplay%20Tag%20Container%20Structure%20%28by%20ref%29) (Momentum :type Vector :default "0, 0, 0") (HitInfo :type Hit%20Result%20Structure%20%28by%20ref%29) (InstigatedBy :type Controller%20Object%20Reference) (DamageCauser :type Actor%20Object%20Reference) (EffectContext :type Gameplay%20Effect%20Context%20Handle%20Structure))
  ;[0 0.5834 1] Trigger the Flipper Events
  (if (Variables|UserOptions|GetbTriggerOnDamage)
    (CallFunction|CalculateLastShotDirection HitInfo)
    (Variables|Default|SetbIsTriggeredLaunch true)
    (FlipperTriggering|AttemptTriggerFlipper true false DamageCauser)))

(event EventPostFinishSpawning
  ;[0 1 0.1154] Initialization
  (|CallParentFunction)
  (CallFunction|Resave))

(event EventBeginPlay
  ;[0 1 0.1154] Initialization
  (|CallParentFunction)
  (Networking|SwitchHasAuthority
    (:Authority
      (Ability|EquipFortAbilitySet self "/CreativeCoreDevices/PinballFlipper/GAS_Creative_PinballFlipper_V2.GAS_Creative_PinballFlipper_V2")
      (Variables|Default|SetFlipper%20Swing%20Time (Components|Timeline|GetTimelineLength (Variables|Device_PinballFlipper_V2|GetFlipperRotation))))
    (:Remote
      _))
  (if (Fortnite|IsPhysicallySimulatedIsland)
    (PhysicsSettings|NewtonBindEvents)
    (FlipperTriggering|DisableFlipper)))

(event Custom|LaunchPhysicsProp
  (params (PhysicsProp :type Actor%20Object%20Reference) (Hit :type Hit%20Result%20Structure))
  (bind _bistriggeredlaunch (Variables|Default|GetbIsTriggeredLaunch))
  (bind _ontriggeredknockbackphysicsprop (Variables|UserOptions|Basic|GetOnTriggeredKnockbackPhysicsProp))
  (bind _onbumpknockbackphysicsprop (Variables|UserOptions|Basic|GetOnBumpKnockbackPhysicsProp))
  (bind _returnvalue (select _bistriggeredlaunch _ontriggeredknockbackphysicsprop _onbumpknockbackphysicsprop))
  (bind _knockupamountphysicsprop (Variables|UserOptions|Basic|GetKnockupAmountPhysicsProp))
  ;[0.9692 1 0] Launch Events
  (Utilities|IsValid PhysicsProp
    (:Is%20Valid
      (CallFunction|PlayKnockbackSound 0 PhysicsProp (Math|Vector2D|MakeVector2D _knockupamountphysicsprop _returnvalue))
      (Class|NewtonPhysicsPropsLibrary|LaunchPhysicsPropswithVector self PhysicsProp (CallFunction|CalculateLaunchVector Hit _knockupamountphysicsprop _returnvalue)))))

(event Custom|Enable%20Event
  (params (Instigator :type Controller%20Object%20Reference))
  ;[0.6337 0 1] Gameplay Functions
  (FlipperTriggering|ResetFlipper))

(event Custom|FlipperTopperDown
  ;[1 1 1] Player Visible Effects
  (if (not (Editor|InEditor))
    (bind (_direction _flippertopperz) (Timeline|FlipperTopDown PlayFromStart :NewTime (Components|Timeline|GetTimelineLength (Variables|Device_PinballFlipper_V2|GetFlipperTopDown)))
      (:Update
        (Transformation|SetRelativeLocation (Variables|Default|GetSM_Pinball_Flipper_Topper) (Math|Vector|MakeVector 0.0 0.0 _flippertopperz) false false)))
    (else
      (Timeline|FlipperTopDown SetNewTime))))

(event Custom|Disable%20Event
  (params (Instigator :type Controller%20Object%20Reference))
  ;[0.6337 0 1] Gameplay Functions
  (FlipperTriggering|DisableFlipper))

(event Custom|Activate%20Event
  (params (Instigator :type Controller%20Object%20Reference))
  ;[0.6337 0 1] Gameplay Functions
  (if (Fortnite|IsPhysicallySimulatedIsland)
    (Class|FortPhysicsImpulseComponent|TriggerDevice (Variables|Default|GetFortPhysicsImpulse))
    (else
      (Variables|TriggerDirection|SetLast%20Shot%20Direction (Math|Random|RandomBool)) ; Triggered by receiver won't have a shot direction, so use a random one.
      (Variables|Default|SetbIsTriggeredLaunch true)
      (FlipperTriggering|AttemptTriggerFlipper true))))

(event Custom|LaunchPhysicsObject
  (params (Target :type Fort%20Physics%20Pawn%20Object%20Object%20Reference) (Hit :type Hit%20Result%20Structure))
  (bind (_vehicleknockup _pawnknockup _vehicleknockback) (CallFunction|GetKnockbackValues))
  ;[0.9692 1 0] Launch Events
  (Utilities|IsValid Target
    (:Is%20Valid
      (CallFunction|PlayKnockbackSound 0 Target (Math|Vector2D|MakeVector2D _vehicleknockup _vehicleknockback))
      (PhysicsObject|ApplyImpulse Target self (CallFunction|CalculateLaunchVector Hit _vehicleknockup _vehicleknockback)))))

(event Custom|FlipperTopperUp
  ;[1 1 1] Player Visible Effects
  (if (not (Editor|InEditor))
    (Audio|PlaySoundatLocation (Variables|Visuals|GetFlipper%20Ready) (Transformation|GetActorLocation))
    (bind (_direction _flippertopperz) (Timeline|FlipperTopperZUp PlayFromStart :NewTime (Components|Timeline|GetTimelineLength (Variables|Device_PinballFlipper_V2|GetFlipperTopperZUp)))
      (:Update
        (Transformation|SetRelativeLocation (Variables|Default|GetSM_Pinball_Flipper_Topper) (Math|Vector|MakeVector 0.0 0.0 _flippertopperz) false false)))
    (else
      (Timeline|FlipperTopperZUp SetNewTime))))

(event Custom|LaunchVehicle
  (params (Vehicle :type Fort%20Athena%20Vehicle%20Object%20Reference) (Hit :type Hit%20Result%20Structure))
  (bind (_vehicleknockup _pawnknockup _vehicleknockback) (CallFunction|GetKnockbackValues))
  ;[0.9692 1 0] Launch Events
  (Score|AwardScoretoAllPassengers Vehicle)
  (CallFunction|PlayKnockbackSound 0 Vehicle (Math|Vector2D|MakeVector2D _vehicleknockup _vehicleknockback))
  (Audio|BroadcastAbilityImpactatLocation 0 Vehicle (CallFunction|CalculateLaunchVector Hit _vehicleknockup _vehicleknockback) 777.0))

(event Custom|TriggerVFX
  ;[1 1 1] Player Visible Effects
  (Networking|CanExecuteCosmeticEvents
    (:True
      (Utilities|FlowControl|Sequence
        (:then_0
          (bind _returnvalue (Effects|SpawnEmitteratLocation "/CreativeCoreDevices/Athena/Effects/Particles/Creative/Props/P_Creative_Pinball_Flipper.P_Creative_Pinball_Flipper" (Transformation|GetWorldLocation (Variables|Default|GetFlipperStaticMesh))))
          (Effects|Components|ParticleSystem|SetColorParameter _returnvalue "LightColor" (Utilities|Map|Find (Variables|UserKnobs|GetFX_Colors) (Variables|UserOptions|Basic|GetFlipperColor))))
        (:then_1
          (Audio|PlaySoundatLocation (Variables|Sound|GetFlipper%20Sound) (Transformation|GetActorLocation))
          (if (not (Fortnite|IsPhysicallySimulatedIsland))
            (if (!= (.RawMagnitude (Ability|GameplayCue|BreakGameplayCueParameters (stale GameplayCue|GameplayCue.Abilities.Activation.Devices.PinballFlipper.Hit Parameters))) 0.0)
              (bind (_direction _flipperzrotation) (Timeline|FlipperRotation PlayFromStart)
                (:Update
                  (Transformation|SetRelativeRotation (Variables|Default|GetFlipperStaticMesh) (Math|Rotator|MakeRotator 0.0 0.0 (* _flipperzrotation (.RawMagnitude (Ability|GameplayCue|BreakGameplayCueParameters (stale GameplayCue|GameplayCue.Abilities.Activation.Devices.PinballFlipper.Hit Parameters))))) false false))))))
        (:then_2
          (CallFunction|SwitchToDyamicMaterial)
          (bind _material_instance_dynamic (Variables|Visuals|GetMaterial%20Instance%20Dynamic)
            (:then
              (Rendering|Material|SetScalarParameterValue _material_instance_dynamic "GlowOn" 1.0 :!class /Script/Engine.MaterialInstanceDynamic)
              (Utilities|Time|SetTimerbyEvent (delegate Custom|ResetMaterial) 0.2))
            (:else)))))))

(event GameplayCue|GameplayCue.Abilities.Activation.Devices.PinballFlipper.Hit
  (params (EventType :type EGameplayCueEvent%20Enum :default "OnActive") (Parameters :type Gameplay%20Cue%20Parameters%20Structure%20%28by%20ref%29))
  (bind (_normalizedmagnitude _rawmagnitude) (Ability|GameplayCue|BreakGameplayCueParameters Parameters))
  ;[1 1 1] Player Visible Effects
  (Networking|CanExecuteCosmeticEvents
    (:True
      (Utilities|FlowControl|Sequence
        (:then_0
          (bind _returnvalue (Effects|SpawnEmitteratLocation "/CreativeCoreDevices/Athena/Effects/Particles/Creative/Props/P_Creative_Pinball_Flipper.P_Creative_Pinball_Flipper" (Transformation|GetWorldLocation (Variables|Default|GetFlipperStaticMesh))))
          (Effects|Components|ParticleSystem|SetColorParameter _returnvalue "LightColor" (Utilities|Map|Find (Variables|UserKnobs|GetFX_Colors) (Variables|UserOptions|Basic|GetFlipperColor))))
        (:then_1
          (Audio|PlaySoundatLocation (Variables|Sound|GetFlipper%20Sound) (Transformation|GetActorLocation))
          (if (not (Fortnite|IsPhysicallySimulatedIsland))
            (if (!= _rawmagnitude 0.0)
              (Timeline|FlipperRotation PlayFromStart))))
        (:then_2
          (CallFunction|SwitchToDyamicMaterial)
          (bind _material_instance_dynamic (Variables|Visuals|GetMaterial%20Instance%20Dynamic)
            (:then
              (Rendering|Material|SetScalarParameterValue _material_instance_dynamic "GlowOn" 1.0 :!class /Script/Engine.MaterialInstanceDynamic)
              (Utilities|Time|SetTimerbyEvent (delegate Custom|ResetMaterial) 0.2))
            (:else)))))))

(event Custom|OnComponentHit_Event
  (params (HitComponent :type Primitive%20Component%20Object%20Reference) (OtherActor :type Actor%20Object%20Reference) (OtherComp :type Primitive%20Component%20Object%20Reference) (NormalImpulse :type Vector) (Hit :type Hit%20Result%20Structure%20%28by%20ref%29))
  ;[0.006422 0.15 0 0.5] Newton Specific
  ;;[1 1 1] When a pawn is hit - we need to use Launch Character in the game thread instead of triggering from the physic thread
  (if (not (Variables|Default|GetbIsDisabled))
    (if (> (Math|Vector|VectorLength (Physics|GetPhysicsAngularVelocityinDegrees (Variables|Default|GetFlipperStaticMesh) "None")) (Variables|Default|GetPawnPushThreshold))
      (bind _asplayer_pawn_athena (Utilities|Casting|CastToPlayerPawn_Athena OtherActor))
      (CallFunction|SetUpFlipperVectorCachedPerPlayer _asplayer_pawn_athena)
      (CallFunction|LaunchPawn _asplayer_pawn_athena Hit))))

(event Custom|LaunchPawn
  (params (PlayerPawn :type Fort%20Pawn%20Object%20Reference) (Hit :type Hit%20Result%20Structure))
  (bind (_vehicleknockup _pawnknockup _vehicleknockback _pawnknockback) (CallFunction|GetKnockbackValues))
  ;[0.9692 1 0] Launch Events
  ;[0.006422 0.15 0 0.5] Newton Specific
  (Score|AwardScore PlayerPawn)
  (CallFunction|PlayKnockbackSound 0 PlayerPawn (Math|Vector2D|MakeVector2D _pawnknockup _pawnknockback))
  (Pawn|Character|LaunchCharacterJump PlayerPawn (CallFunction|CalculatePlayerLaunchVector Hit _pawnknockup _pawnknockback PlayerPawn) true true (not (Variables|UserOptions|Basic|GetbFallDamage)) false))

(event Custom|OnEndPlay_Event
  (params (Actor :type Actor%20Object%20Reference) (EndPlayReason :type EEndPlayReason%20Enum :default "Destroyed"))
  ;[0.006422 0.15 0 0.5] Newton Specific
  ;;[1 1 1] Physic component setup and subscribing to activation event for triggering VFX
  (EventDispatchers|UnbindallEventsfromOnPhysicsImpulseOverlapTriggeredSignatureDelegate (Variables|Default|GetFortPhysicsImpulse))
  (Collision|UnbindallEventsfromOnComponentHit (Variables|Default|GetFlipperStaticMesh)))

(event Custom|ResetMaterial
  ;[1 1 1] Player Visible Effects
  (bind _material_instance_dynamic (Variables|Visuals|GetMaterial%20Instance%20Dynamic)
    (:then
      (Rendering|Material|SetScalarParameterValue _material_instance_dynamic "GlowOn" 0.0 :!class /Script/Engine.MaterialInstanceDynamic)
      (CallFunction|SwitchToConstantMaterial))
    (:else)))

(event Custom|OnPhysicsImpulseOverlapTriggeredSignatureDelegate_Event
  (params (Payload :type Fort%20Physics%20Impulse%20Component%20Delegate%20Payload%20Structure))
  ;[0.006422 0.15 0 0.5] Newton Specific
  ;;[1 1 1] Physic component setup and subscribing to activation event for triggering VFX
  (if (not (Variables|Default|GetbIsDisabled))
    (CallFunction|TriggerVFX)
    (Networking|SwitchHasAuthority
      (:Authority
        (FlipperTriggering|SetupReset (Variables|UserOptions|Basic|GetFlipper%20Velocity))
        (GameplayEvent|BroadcastEvent (Variables|UserOptions|Events|GetOn%20Activated) 0 (Utilities|Struct|MakeGameplayEventActivated 0))))))

(event Custom|LoadMaterials
  (bind _flippercolor (Variables|UserOptions|Basic|GetFlipperColor))
  (bind _flippermaterials (Variables|UserKnobs|GetFlipperMaterials))
  (bind (_value _returnvalue) (Utilities|Map|Find _flippermaterials _flippercolor))
  (bind _output (Utilities|ResolveSoftReference _value))
  ;[1 1 1] Player Visible Effects
  (if _returnvalue
    (if (Utilities|IsValidSoftObjectReference _value)
      (Utilities|IsValid _output
        (:Is%20Valid
          (CallFunction|UpdateParentMaterial _output))
        (:Is%20Not%20Valid
          (bind _loaded_object (Creative|Loading|AsyncLoadEditorSafe _value)
            (:Loaded
              (bind _asmaterial_interface (Utilities|Casting|CastToMaterialInterface _loaded_object))
              (CallFunction|UpdateParentMaterial _asmaterial_interface))))))))

(event Custom|ResetMapFlipperStaticmeshRightVectorPerPlayerEvent
  ;[0.006422 0.15 0 0.5] Newton Specific
  (Utilities|Map|Clear (Variables|Default|GetMapFlipperStaticMeshRightVectorPerPlayer)))

(event Custom|Switch%20to%20Dyamic%20Material
  ;[1 1 1] Player Visible Effects
  (Networking|CanExecuteCosmeticEvents
    (:True
      (Variables|Visuals|GetMaterial%20Instance%20Dynamic
        (:then
          _)
        (:else
          (bind _material_cache (Variables|Visuals|GetMaterial%20Cache)
            (:then
              (bind (_outmaterialinstance _returnvalue) (Class|FortMaterialCache|ReserveMaterial _material_cache))
              (Variables|Visuals|SetMaterial%20Instance%20Dynamic _outmaterialinstance)
              (if _returnvalue
                _))
            (:else))))
      (Rendering|Material|SetMaterial (Variables|Default|GetFlipperStaticMesh) 0 (Variables|Visuals|GetMaterial%20Instance%20Dynamic)))))

(event Custom|Switch%20to%20Constant%20Material
  ;[1 1 1] Player Visible Effects
  (Networking|CanExecuteCosmeticEvents
    (:True
      (Rendering|Material|SetMaterial (Variables|Default|GetFlipperStaticMesh) 0 (Variables|Visuals|GetParentMaterial))
      (bind _material_cache (Variables|Visuals|GetMaterial%20Cache)
        (:then
          (Class|FortMaterialCache|ReleaseMaterial _material_cache (Variables|Visuals|GetMaterial%20Instance%20Dynamic)))
        (:else)))))

(fn AttemptTriggerFlipper
  (params (UseLargeVolume :type Boolean) (Trigger%20Clockwise :type Boolean) (Triggering%20Actor :type Actor%20Object%20Reference))
  (Utilities|CanTrigger Triggering%20Actor
    (:%20%20
      (LocalVariables|Setlocal%20Large%20Volume UseLargeVolume)
      (bind (_valid _clockwise_rotation) (FlipperTriggering|DetermineFlipperDirection Trigger%20Clockwise))
      (if _valid
        (bind _output_get (LocalVariables|Setlocal%20Trigger%20Clockwise _clockwise_rotation))
        (Variables|TriggerDirection|SetFlipFlopDirection (not _output_get))
        (FlipperTriggering|TriggerFlipper (LocalVariables|Getlocal%20Large%20Volume) (LocalVariables|Getlocal%20Trigger%20Clockwise) true)
        (bind _currentminigame (Variables|/Script/FortniteGame.FortMinigameLogicComponent|GetCurrentMinigame (Variables|Default|GetFortMinigameProgress))
          (:then
            (Minigame|IsInProgress _currentminigame
              (:Yes
                (Utilities|FlowControl|Sequence
                  (:then_0
                    (bind _ascontroller (Utilities|Casting|CastToController Triggering%20Actor)
                      (:then
                        (LocalVariables|SetInstigatorController _ascontroller))
                      (:CastFailed
                        (bind _aspawn (Utilities|Casting|CastToPawn Triggering%20Actor)
                          (:then
                            (LocalVariables|SetInstigatorController (Pawn|GetController _aspawn)))
                          (:CastFailed
                            ;[1 1 1] Vehicle
                            (bind _asfort_vehicle_interface (Utilities|Casting|CastToFortVehicleInterface Triggering%20Actor))
                            (LocalVariables|SetInstigatorController (Class|FortVehicleInterface|GetVehicleController _asfort_vehicle_interface (Vehicle|GetDriver _asfort_vehicle_interface))))))))
                  (:then_1
                    (GameplayEvent|BroadcastEvent (Variables|UserOptions|Events|GetOn%20Activated) 0 (Utilities|Struct|MakeGameplayEventActivated (LocalVariables|GetInstigatorController))))))))
          (:else))))))

(fn AttemptBackSwingAfterDelay
  (params (TriggerDirection :type Float%20%28double-precision%29))
  (if (and (Variables|UserOptions|GetbHitOnBackswing) (!= TriggerDirection 0.0))
    (Utilities|Time|ClearandInvalidateTimerbyHandle (Variables|Default|GetBackswing%20Timer))
    (bind _returnvalue (Utilities|Time|SetTimerbyFunctionName 0 "BackSwingHit" (Variables|Default|GetFlipper%20Swing%20Time)))
    (Variables|Default|SetBackswing%20Timer _returnvalue)
    (else
      (Collision|SetCollisionEnabled (Variables|Default|GetClockwiseTrigger) "NoCollision"))))

(fn AwardScore
  (params (Actor :type Actor%20Object%20Reference))
  (Networking|SwitchHasAuthority
    (:Authority
      (bind _output_get (LocalVariables|Setlocal%20PC (CreativeAI|Controller|GetParticipantControllerfromActor Actor false)))
      (Utilities|IsValid _output_get
        (:Is%20Valid
          (bind _returnvalue (CreativeDevices|GetMinigameStatsComp (LocalVariables|Getlocal%20PC) "Athena"))
          (LocalVariables|Setlocal%20MinigameStatsComp _returnvalue)
          (bind _local_minigamestatscomp (LocalVariables|Getlocal%20MinigameStatsComp)
            (:then
              (Class|FortMinigameStatsComponent|SendStatEventwithTags _local_minigamestatscomp (LocalVariables|Getlocal%20PC) "MinigameDynamicEvent" self (GameplayTags|MakeGameplayTagContainerfromTag "(TagName=\"Minigame.Stat.Score.Pinball\")") (Class|FortMinigameStatsComponent|GetSourceandContextTags _local_minigamestatscomp (LocalVariables|Getlocal%20PC)) (.OutContextTags (Class|FortMinigameStatsComponent|GetSourceandContextTags _local_minigamestatscomp (LocalVariables|Getlocal%20PC))) (Variables|UserOptions|Basic|GetScoreValue))
              (Class|CreativeHUDBroadcastComponent|DoScoreEventforParticipant (Variables|Default|GetCreative_HUDBroadcast_Component) (LocalVariables|Getlocal%20PC) (Variables|UserOptions|Basic|GetScoreValue) (Variables|UserOptions|Advanced|GetDisplayScoreUpdateOnHUD)))
            (:else)))))))

(fn AwardScoreToAllPassengers
  (params (Target :type Fort%20Vehicle%20Interface%20Interface))
  (Networking|SwitchHasAuthority
    (:Authority
      (for _array_element (Vehicle|GetAllPassengers Target)
        (Utilities|IsValid _array_element
          (:Is%20Valid
            (Score|AwardScore _array_element)))))))

(fn BackSwingHit
  (Collision|SetCollisionEnabled (Variables|Default|GetClockwiseTrigger) "QueryOnly")
  (FlipperTriggering|TriggerFlipper false (Variables|TriggerDirection|GetFlipFlopDirection)))

(fn BlueprintCanInteract
  (params (InteractingPawn :type Fort%20Pawn%20Object%20Reference) (InteractionBeingAttempted :type EInteractionBeingAttempted%20Enum :default "FirstInteraction") (InteractionType :type TInteractionType%20Enum :default "IT_NoInteraction"))
  (bind _returnvalue (Pawn|GetController InteractingPawn))
  (bind _fortminigameprogress (Variables|Default|GetFortMinigameProgress))
  (bind _currentminigame (Variables|/Script/FortniteGame.FortMinigameLogicComponent|GetCurrentMinigame _fortminigameprogress))
  (Utilities|FlowControl|Sequence
    (:then_0
      (Utilities|IsValid (Variables|/Script/FortniteGame.FortActorOptionsComponent|GetInternal_PlayerOptions (Variables|Default|GetToyOptionsComponent))
        (:Is%20Valid
          (Utilities|IsValid InteractingPawn
            (:Is%20Valid
              (Utilities|IsValid _returnvalue
                (:Is%20Valid
                  (bind _asfort_player_controller_athena (Utilities|Casting|CastToFortPlayerControllerAthena _returnvalue))
                  (if (Creative|CanPlayerCreateinCurrentExperience _asfort_player_controller_athena)
                    (Utilities|IsValid _currentminigame
                      (:Is%20Valid
                        (LocalVariables|SetbInteractSuccess (not (Minigame|HasStarted _currentminigame :!class /Script/FortniteGame.FortMinigame))))
                      (:Is%20Not%20Valid
                        (LocalVariables|SetbInteractSuccess true)))))))))))
    (:then_1
      (return (LocalVariables|GetbInteractSuccess)))))

(fn BlueprintGetInteractionString
  (params (InteractingPawn :type Fort%20Pawn%20Object%20Reference) (InteractionBeingAttempted :type EInteractionBeingAttempted%20Enum :default "FirstInteraction"))
  (return "NSLOCTEXT(\"[6B5FEC2E0EAC7B58C8E8DCAB79558E3E]\", \"3250428B4A54123A50BC22824E6A9928\", \"Customize\")"))

(fn BlueprintModifyIncomingDamage
  (params (Damage :type Float%20%28single-precision%29) (InTags :type Gameplay%20Tag%20Container%20Structure%20%28by%20ref%29) (EffectContext :type Gameplay%20Effect%20Context%20Handle%20Structure%20%28by%20ref%29) (EventInstigator :type Controller%20Object%20Reference) (DamageCauser :type Actor%20Object%20Reference))
  (return 0.0))

(fn BlueprintOnLocalInteract
  (params (InteractingPawn :type Fort%20Player%20Pawn%20Object%20Reference) (InteractionBeingAttempted :type EInteractionBeingAttempted%20Enum :default "FirstInteraction"))
  (return false))

(fn CalculateLastShotDirection
  (params (Hit :type Hit%20Result%20Structure))
  (bind (_bblockinghit _binitialoverlap _time _distance _location _impactpoint _normal _impactnormal) (Collision|BreakHitResult Hit))
  (bind (_returnvalue_x _returnvalue_y) (Math|Vector|BreakVector (Math|Vector|UnrotateVector _impactnormal (Transformation|GetWorldRotation (Variables|Default|GetFlipperStaticMesh)))))
  (Variables|TriggerDirection|SetLast%20Shot%20Direction (> _returnvalue_y 0.0)))

(fn CalculateLaunchVector
  (params (HitData :type Hit%20Result%20Structure) (KnockupValue :type Float%20%28double-precision%29) (KnockbackValue :type Float%20%28double-precision%29))
  (bind _returnvalue (CallFunction|CalculateLunchVectorfromUpandRightVectors HitData KnockupValue KnockbackValue (Transformation|GetUpVector (Variables|Default|GetFlipperStaticMesh)) (Transformation|GetRightVector (Variables|Default|GetFlipperStaticMesh))))
  (return _returnvalue))

(fn CalculateLunchVectorFromUpAndRightVectors
  (params (HitData :type Hit%20Result%20Structure%20%28by%20ref%29) (KnockupValue :type Float%20%28double-precision%29) (KnockbackValue :type Float%20%28double-precision%29) (UpVector :type Vector) (RightVector :type Vector))
  (bind (_bblockinghit _binitialoverlap _time _distance _location _impactpoint) (Collision|BreakHitResult HitData))
  (return (* (Variables|PhysicsSettings|GetPhysics%20Tuning%20Multiplier) (+ (* "0, 0,1.000000" KnockupValue) (* (Math|Vector|Normalize (Math|Vector|Lerp(Vector) (select (> (Math|Vector|DotProduct (- _impactpoint (Transformation|GetWorldLocation (Variables|Default|GetBlockingVolume))) RightVector) 0.0) RightVector (* RightVector -1.0)) (Math|Vector|Normalize (Math|Vector|ProjectVectorontoPlane (- _impactpoint (Transformation|GetWorldLocation (Variables|Default|GetBlockingVolume))) UpVector)) (/ (Variables|UserOptions|GetBounceAnglePercentage) 1e+02))) KnockbackValue)))))

(fn CalculatePlayerLaunchVector
  (params (HitData :type Hit%20Result%20Structure) (KnockupValue :type Float%20%28double-precision%29) (KnockbackValue :type Float%20%28double-precision%29) (PlayerToLunch :type Fort%20Pawn%20Object%20Reference%20%28by%20ref%29))
  (bind _returnvalue (CallFunction|CalculateLunchVectorfromUpandRightVectors HitData KnockupValue KnockbackValue (Transformation|GetUpVector (Variables|Default|GetFlipperStaticMesh)) (Math|Vector|SelectVector (Utilities|Map|Find (Variables|Default|GetMapFlipperStaticMeshRightVectorPerPlayer) PlayerToLunch) (Transformation|GetRightVector (Variables|Default|GetFlipperStaticMesh)) (Fortnite|IsPhysicallySimulatedIsland))))
  (return _returnvalue))

(fn DetermineFlipperDirection
  (params (Trigger%20Clockwise :type Boolean))
  (LocalVariables|Setlocal%20Attempt%20Clockwise Trigger%20Clockwise)
  (if (Variables|Default|GetbIsTriggeredLaunch)
    (return true (Utilities|Select :Index (Variables|UserOptions|Basic|GetFlipDirection) :NewEnumerator0 false :NewEnumerator1 false :NewEnumerator5 true :NewEnumerator2 (not (Variables|TriggerDirection|GetLast%20Shot%20Direction)) :NewEnumerator3 (Variables|TriggerDirection|GetLast%20Shot%20Direction) :NewEnumerator4 (Variables|TriggerDirection|GetFlipFlopDirection)))
    (else
      (return (Utilities|Select :Index (Variables|UserOptions|Basic|GetFlipDirection) :NewEnumerator0 (not (LocalVariables|Getlocal%20Attempt%20Clockwise)) :NewEnumerator1 true :NewEnumerator5 (LocalVariables|Getlocal%20Attempt%20Clockwise) :NewEnumerator2 true :NewEnumerator3 true :NewEnumerator4 (Math|Boolean|Equal(Boolean) (LocalVariables|Getlocal%20Attempt%20Clockwise) (Variables|TriggerDirection|GetFlipFlopDirection))) (Utilities|Select :Index (Variables|UserOptions|Basic|GetFlipDirection) :NewEnumerator0 false :NewEnumerator1 false :NewEnumerator5 true :NewEnumerator2 (not (LocalVariables|Getlocal%20Attempt%20Clockwise)) :NewEnumerator3 (LocalVariables|Getlocal%20Attempt%20Clockwise) :NewEnumerator4 (LocalVariables|Getlocal%20Attempt%20Clockwise))))))

(fn DisableFlipper
  (Utilities|FlowControl|Sequence
    (:then_0
      (if (not (Fortnite|IsPhysicallySimulatedIsland))
        (Collision|SetCollisionEnabled (Variables|Default|GetClockwiseTrigger) "NoCollision")
        (else
          (CallFunction|EnableDevicePhysics false))))
    (:then_1
      (Variables|Default|SetbIsDisabled true))))

(fn EnableonStart
  (if (not (Fortnite|IsPhysicallySimulatedIsland))
    (Networking|SwitchHasAuthority
      (:Authority
        _))
    (else
      _))
  (if (Variables|UserOptions|GetbEnabledAtGameStart)
    (FlipperTriggering|ResetFlipper)
    (else
      (FlipperTriggering|DisableFlipper))))

(fn EnableDevicePhysics
  (params (bEnabled :type Boolean))
  (Class|FortPhysicsImpulseComponent|SetImpulseDeviceInputPinballFlipper (Variables|Default|GetFortPhysicsImpulse) "PinballFlipper" (Utilities|Struct|MakeFortPhysicsPinballFlipperDeviceInput (Variables|UserOptions|Basic|GetFlipper%20Velocity) (Math|Integer|BitwiseOR (Math|Integer|BitwiseOR (Math|Integer|BitwiseOR (Math|Integer|BitwiseOR (Math|Integer|BitwiseOR (Math|Integer|BitwiseOR (Math|Integer|BitwiseOR 0 (* (Math|Conversions|ToInteger(Boolean) (Variables|PhysicsSettings|GetAllowPlayers)) (Math|Conversions|ToInteger(Byte) (Utilities|Enum|LiteralenumEAllowedToBounce "Players")))) (* (Math|Conversions|ToInteger(Boolean) (Variables|PhysicsSettings|GetAllowGuards)) (Math|Conversions|ToInteger(Byte) (Utilities|Enum|LiteralenumEAllowedToBounce "Guards")))) (* (Math|Conversions|ToInteger(Boolean) (Variables|PhysicsSettings|GetAllowWildlife)) (Math|Conversions|ToInteger(Byte) (Utilities|Enum|LiteralenumEAllowedToBounce "Wildlife")))) (* (Math|Conversions|ToInteger(Boolean) (Variables|PhysicsSettings|GetAllowCreatures)) (Math|Conversions|ToInteger(Byte) (Utilities|Enum|LiteralenumEAllowedToBounce "Creatures")))) (* (Math|Conversions|ToInteger(Boolean) (Variables|PhysicsSettings|GetAllowVehicles)) (Math|Conversions|ToInteger(Byte) (Utilities|Enum|LiteralenumEAllowedToBounce "Vehicles")))) (* (Math|Conversions|ToInteger(Boolean) (Variables|PhysicsSettings|GetAllowObjects)) (Math|Conversions|ToInteger(Byte) (Utilities|Enum|LiteralenumEAllowedToBounce "Objects")))) (* (Math|Conversions|ToInteger(Boolean) (Variables|PhysicsSettings|GetAllowDBNO)) (Math|Conversions|ToInteger(Byte) (Utilities|Enum|LiteralenumEAllowedToBounce "DBNO")))) bEnabled (Variables|Default|GetFlipperStaticMesh) (Utilities|Enum|BytetoEnumEPinballFlipperDirection (Math|Byte|MakeLiteralByte (Variables|UserOptions|Basic|GetFlipDirection))) (Networking|IsDedicatedServer))))

(fn GetKnockbackValues
  (if (Utilities|Enum|Equal(Enum) (Variables|UserOptions|Basic|GetFlipDirection) "NewEnumerator1")
    (return 0.0 0.0 0.0 0.0)
    (else
      (return (Variables|UserOptions|Basic|GetKnockupAmountVehicle) (Variables|UserOptions|Basic|GetKnockupAmountPlayer) (select (Variables|Default|GetbIsTriggeredLaunch) (Variables|UserOptions|Basic|GetOnTriggeredKnockbackVehicle) (Variables|UserOptions|Basic|GetOnBumpKnockbackVehicle)) (select (Variables|Default|GetbIsTriggeredLaunch) (Variables|UserOptions|Basic|GetOnTriggeredKnockbackPlayer) (Variables|UserOptions|Basic|GetOnBumpKnockbackPlayer))))))

(fn GetOverrideMeshMaterial
  (params (MeshComp :type Static%20Mesh%20Component%20Object%20Reference) (MatIdx :type Integer))
  (if (and (or (Utilities|Equal(Object) (Variables|Default|GetFlipperStaticMesh) MeshComp) (Utilities|Equal(Object) (Variables|Default|GetSM_Pinball_Flipper_Topper) MeshComp)) (== MatIdx 0))
    (bind _material_instance_dynamic (Variables|Visuals|GetMaterial%20Instance%20Dynamic)
      (:then
        (return _material_instance_dynamic))
      (:else
        (return (Variables|Visuals|GetParentMaterial))))
    (else
      (return 0))))

(fn IsTargetableByLinearMover
  (return true))

(fn MessageScore
  (return (Variables|UserOptions|Advanced|GetHUD%20Message) (Variables|UserOptions|Advanced|GetHUD%20MessageScoreColor) (Variables|UserOptions|Advanced|GetHUD%20MessageColor) (Variables|UserOptions|Advanced|GetResetHUD%20MessageScore)))

(fn MigrateLegacyUserOptions
  (bind (_ontriggeredknockbackphysicsprop _onbumpknockbackphysicsprop _knockupamountphysicsprop _rotationdirection) (Class|GetClassDefaults "/CreativeCoreDevices/Device_PinballFlipper_V2.Device_PinballFlipper_V2_C"))
  (bind (_ontriggeredknockbackphysicsprop_2 _onbumpknockbackphysicsprop_2 _knockupamountphysicsprop_2 _bounceanglealpha) (Class|GetClassDefaults "/CreativeCoreDevices/Device_PinballFlipper_V2.Device_PinballFlipper_V2_C"))
  (bind (_ontriggeredknockbackphysicsprop_3 _onbumpknockbackphysicsprop_3 _knockupamountphysicsprop_3 _knockbacklevelshot) (Class|GetClassDefaults "/CreativeCoreDevices/Device_PinballFlipper_V2.Device_PinballFlipper_V2_C"))
  (bind (_ontriggeredknockbackphysicsprop_4 _onbumpknockbackphysicsprop_4 _knockupamountphysicsprop_4 _knockbacklevel) (Class|GetClassDefaults "/CreativeCoreDevices/Device_PinballFlipper_V2.Device_PinballFlipper_V2_C"))
  (bind (_ontriggeredknockbackphysicsprop_5 _onbumpknockbackphysicsprop_5 _knockupamountphysicsprop_5 _knockupvalueindex) (Class|GetClassDefaults "/CreativeCoreDevices/Device_PinballFlipper_V2.Device_PinballFlipper_V2_C"))
  (bind (_ontriggeredknockbackphysicsprop_6 _onbumpknockbackphysicsprop_6 _knockupamountphysicsprop_6 _damagelevel) (Class|GetClassDefaults "/CreativeCoreDevices/Device_PinballFlipper_V2.Device_PinballFlipper_V2_C"))
  (bind _rotationdirection_2 (Variables|Legacy|GetRotationDirection))
  (bind _knockbacklevel_2 (Variables|Legacy|GetKnockbackLevel))
  (bind _knockbacklevelshot_2 (Variables|Legacy|GetKnockbackLevelShot))
  (bind _knockupvalueindex_2 (Variables|Legacy|GetKnockupValueIndex))
  (bind _bounceanglealpha_2 (Variables|Legacy|GetBounceAngleAlpha))
  (bind _damagelevel_2 (Variables|Legacy|GetDamageLevel))
  (|CallParentFunction)
  (Utilities|FlowControl|Sequence
    (:then_0
      (Utilities|FlowControl|Sequence
        (:then_0
          (if (!= _rotationdirection _rotationdirection_2)
            (Variables|UserOptions|Basic|SetFlipDirection (Utilities|Select :Index (+ (Math|Float|Floor _rotationdirection_2) 1) :Option%200 "NewEnumerator0" :Option%201 "NewEnumerator1" :Option%202 "NewEnumerator5" :Option%203 "NewEnumerator2" :Option%204 "NewEnumerator3" :Option%205 "NewEnumerator4"))
            ;[1 1 1] Now that we've used the non-default values from the original device, update the value to the defalut and write that to this device so we don't run this migrate logic again
            (Variables|Legacy|SetRotationDirection _rotationdirection)
            (LocalVariables|SetMigrated true)))
        (:then_1
          (if (!= _knockbacklevelshot _knockbacklevelshot_2)
            (Variables|UserOptions|Basic|SetOnTriggeredKnockbackVehicle (Utilities|Select :Index _knockbacklevelshot_2 :Option%200 0.0 :Option%201 4e+02 :Option%202 8e+02 :Option%203 1.2e+03 :Option%204 2e+03 :Option%205 5e+03 :Option%206 7e+04))
            (Variables|UserOptions|Basic|SetOnTriggeredKnockbackPlayer (Utilities|Select :Index _knockbacklevelshot_2 :Option%200 0.0 :Option%201 8e+02 :Option%202 1.6e+03 :Option%203 2.4e+03 :Option%204 4e+03 :Option%205 7.5e+03 :Option%206 1e+04))
            (Variables|Legacy|SetKnockbackLevelShot _knockbacklevelshot)
            (LocalVariables|SetMigrated true)))
        (:then_2
          (if (!= _knockbacklevel _knockbacklevel_2)
            (Variables|UserOptions|Basic|SetOnBumpKnockbackVehicle (Utilities|Select :Index _knockbacklevel_2 :Option%200 0.0 :Option%201 4e+02 :Option%202 8e+02 :Option%203 1.2e+03 :Option%204 2e+03 :Option%205 5e+03 :Option%206 7e+04))
            (Variables|UserOptions|Basic|SetOnBumpKnockbackPlayer (Utilities|Select :Index _knockbacklevel_2 :Option%200 0.0 :Option%201 8e+02 :Option%202 1.6e+03 :Option%203 2.4e+03 :Option%204 4e+03 :Option%205 7.5e+03 :Option%206 1e+04))
            (Variables|Legacy|SetKnockbackLevel _knockbacklevel)
            (LocalVariables|SetMigrated true)))
        (:then_3
          (if (!= _knockupvalueindex _knockupvalueindex_2)
            (Variables|UserOptions|Basic|SetKnockupAmountVehicle (Utilities|Select :Index _knockupvalueindex_2 :Option%200 0.0 :Option%201 4e+02 :Option%202 6e+02 :Option%203 9e+02 :Option%204 1.2e+03 :Option%205 1.5e+03 :Option%206 1.8e+03))
            (Variables|UserOptions|Basic|SetKnockupAmountPlayer (Utilities|Select :Index _knockupvalueindex_2 :Option%200 0.0 :Option%201 4e+02 :Option%202 8e+02 :Option%203 1.2e+03 :Option%204 1.6e+03 :Option%205 2e+03 :Option%206 2.4e+03))
            (Variables|Legacy|SetKnockupValueIndex _knockupvalueindex)
            (LocalVariables|SetMigrated true)))
        (:then_4
          (if (!= _bounceanglealpha _bounceanglealpha_2)
            (Variables|UserOptions|SetBounceAnglePercentage (* _bounceanglealpha_2 1e+02))
            (Variables|Legacy|SetBounceAngleAlpha _bounceanglealpha)
            (LocalVariables|SetMigrated true)))
        (:then_5
          (if (!= _damagelevel _damagelevel_2)
            (Variables|UserOptions|SetDamage (Utilities|Select :Index _damagelevel_2 :Option%200 0 :Option%201 1 :Option%202 2 :Option%203 5 :Option%204 10 :Option%205 20 :Option%206 50 :Option%207 100 :Option%208 5000))
            (Variables|Legacy|SetDamageLevel _damagelevel)
            (LocalVariables|SetMigrated true)))))
    (:then_1
      ;[1 1 1] Commit the changes so it's saved next time it loads and the logic isn't run again
      (if (LocalVariables|GetMigrated)
        (OptionData|RefreshPropertyOverrides (Variables|Default|GetToyOptionsComponent))))))

(fn Newton_BindEvents
  (Collision|BindEventtoOnComponentHit (Variables|Default|GetFlipperStaticMesh) (EventDispatchers|CreateEvent :!event OnComponentHit_Event))
  (EventDispatchers|BindEventtoOnPhysicsImpulseOverlapTriggeredSignatureDelegate (Variables|Default|GetFortPhysicsImpulse) (EventDispatchers|CreateEvent :!event OnPhysicsImpulseOverlapTriggeredSignatureDelegate_Event))
  (Game|BindEventtoOnEndPlay (EventDispatchers|CreateEvent :!event OnEndPlay_Event))
  (EventDispatchers|BindEventtoOnFlipperMovementRestartCompleteSignatureDelegate (Variables|Default|GetFortPhysicsImpulse) (EventDispatchers|CreateEvent :!event ResetMapFlipperStaticmeshRightVectorPerPlayerEvent)))

(fn Newton_Teardown
  (Physics|SetSimulatePhysics (Variables|Default|GetFlipperStaticMesh) false)
  (bind _physicconstraintcomponent (Variables|Default|GetPhysicConstraintComponent)
    (:then
      (Physics|Components|PhysicsConstraint|BreakConstraint _physicconstraintcomponent)
      (Components|DestroyComponent _physicconstraintcomponent))
    (:else
      _))
  (Transformation|AttachComponentToComponent (Variables|Default|GetFlipperStaticMesh) (Variables|Debug|GetStaticMeshComponent) "None" "SnapToTarget" "SnapToTarget" "KeepRelative" true))

(fn NewtonSetup
  (bind _physicconstraintcomponent (Variables|Default|GetPhysicConstraintComponent))
  ;[1 0.313 0 0.5] Never Newton in a Preview
  (if (not (Building|IsPreviewBuildingActor))
    ;[1 1 1] We disable the standard trigger logic with this
    (Variables|PhysicsSettings|SetPhysics%20Tuning%20Multiplier (ScalableFloat|GetValueasDouble (Variables|PhysicsSettings|GetHF_PhysicsTuningMultiplier)))
    (Collision|SetCollisionEnabled (Variables|Default|GetClockwiseTrigger) "NoCollision")
    (Variables|Default|SetbResetEnabled false) ; Is this variable really used?
    (Utilities|FlowControl|Sequence
      (:then_0
        ;[1 1 1] Setup swing arc value or double rotation for the different rotation modes
        (switch Utilities|FlowControl|Switch|SwitchonEnum_PinballFlipDirection (Variables|UserOptions|Basic|GetFlipDirection)
          (:NewEnumerator0
            (Variables|UserOptions|Basic|SetSwing%20Arc (* (Variables|UserOptions|Basic|GetSwing%20Arc) -1.0)))
          (:NewEnumerator2
            _)
          (:NewEnumerator3
            _)
          (:NewEnumerator4
            _))
        (LocalVariables|SetbDoubleDirection true))
      (:then_1
        ;[1 1 1] Create and configure a physic constraint for the flipper
        (Variables|Default|GetPhysicConstraintComponent
          (:then
            _)
          (:else
            (bind _returnvalue (AddComponent|NODE_AddPhysicsConstraintComponent-0 :bManualAttachment true))
            (Variables|Default|SetPhysicConstraintComponent _returnvalue)))
        (Transformation|AttachComponentToComponent (Variables|Default|GetPhysicConstraintComponent) (Variables|Debug|GetStaticMeshComponent) "None" "KeepRelative" "KeepRelative" "KeepRelative" true)
        (Transformation|SetRelativeLocation _physicconstraintcomponent "0.000000,0.000000,32.000000" false true)
        (Physics|Components|PhysicsConstraint|SetProjectionEnabled _physicconstraintcomponent false)
        (Physics|Components|PhysicsConstraint|SetLinearXLimit _physicconstraintcomponent "LCM_Locked" 0.0)
        (Physics|Components|PhysicsConstraint|SetLinearYLimit _physicconstraintcomponent "LCM_Locked" 0.0)
        (Physics|Components|PhysicsConstraint|SetLinearZLimit _physicconstraintcomponent "LCM_Locked" 0.0)
        (Physics|Components|PhysicsConstraint|SetAngularSwing2Limit _physicconstraintcomponent "ACM_Locked" 0.0)
        (Physics|Components|PhysicsConstraint|SetAngularTwistLimit _physicconstraintcomponent "ACM_Locked" 0.0)
        (Physics|Components|PhysicsConstraint|SetAngularDriveMode _physicconstraintcomponent "TwistAndSwing")
        (Physics|Components|PhysicsConstraint|SetDisableCollision _physicconstraintcomponent true)
        (Physics|Components|PhysicsConstraint|SetOrientationDriveTwistandSwing _physicconstraintcomponent false true)
        (Physics|Components|PhysicsConstraint|SetAngularVelocityDriveTwistandSwing _physicconstraintcomponent false true))
      (:then_2
        ;[1 1 1] User settings for the Physic constraint and flipper mass
        (Transformation|SetRelativeRotation (Variables|Default|GetFlipperStaticMesh) (Math|Rotator|MakeRotator 0.0 0.0 (select (LocalVariables|GetbDoubleDirection) 0.0 (/ (Variables|UserOptions|Basic|GetSwing%20Arc) 2.0))) false false)
        (Physics|SetSimulatePhysics (Variables|Default|GetFlipperStaticMesh) true)
        ;;[1 0 0 0.5] We only create and setup the constraint if it doesn't exists (It was destroyed at Teadown)
        ;; Make sure its only setup once (This method will be called several times, we only setup once)
        (bind _outcomponent1 (Physics|Components|PhysicsConstraint|GetConstrainedComponents (Variables|Default|GetPhysicConstraintComponent)))
        (Utilities|IsValid _outcomponent1
          (:Is%20Not%20Valid
            (Physics|Components|PhysicsConstraint|SetConstrainedComponents (Variables|Default|GetPhysicConstraintComponent) (Variables|Default|GetFlipperStaticMesh) "None" 0 "None")
            (Physics|Components|PhysicsConstraint|SetAngularSwing1Limit (Variables|Default|GetPhysicConstraintComponent) "ACM_Limited" (Math|Float|Absolute(Float) (select (LocalVariables|GetbDoubleDirection) (Variables|UserOptions|Basic|GetSwing%20Arc) (/ (Variables|UserOptions|Basic|GetSwing%20Arc) 2.0))))
            (Physics|Components|PhysicsConstraint|SetAngularOrientationTarget (Variables|Default|GetPhysicConstraintComponent) (Math|Rotator|MakeRotator 0.0 0.0 (* (select (LocalVariables|GetbDoubleDirection) 0.0 (/ (Variables|UserOptions|Basic|GetSwing%20Arc) 2.0)) -1.0)))
            (Physics|Components|PhysicsConstraint|SetAngularDriveParams (Variables|Default|GetPhysicConstraintComponent) (Variables|UserOptions|Basic|GetReturn%20Strength) (select (LocalVariables|GetbDoubleDirection) 7.0 0.0) 0.0)
            (Physics|SetMassOverrideinKg (Variables|Default|GetFlipperStaticMesh) "None" (Variables|UserOptions|Basic|GetFlipper%20Mass) true))))
      (:then_3
        (CallFunction|EnableDevicePhysics (and (Variables|UserOptions|GetbTriggerOnProximity) (Variables|UserOptions|GetbEnabledAtGameStart)))))))

(fn OnRep_bIsResetting
  (if (Variables|Default|GetbIsDisabled)
    (CallFunction|FlipperTopperDown)
    (else
      (CallFunction|FlipperTopperUp))))

(fn OnRep_MaterialIndex
  (CallFunction|LoadMaterials))

(fn PlayKnockbackSound
  (params (AttachToComponent :type Actor%20Object%20Reference) (KnockBack%20Value :type Vector%202D%20Structure))
  (Audio|SpawnSoundAttached (select (> (Math|Vector2D|Vector2DLengthSquared KnockBack%20Value) (Math|Float|Square (Variables|Sound|GetPushback%20Fast%20Audio%20Threshold))) (Variables|Sound|GetPushback%20Fast%20Sound) (Variables|Sound|GetPushback%20Slow%20Sound)) (Variables|/Script/Engine.Actor|GetRootComponent AttachToComponent)))

(fn ResetFlipper
  (Utilities|FlowControl|Sequence
    (:then_0
      (if (not (Fortnite|IsPhysicallySimulatedIsland))
        (if (Variables|Default|GetbResetEnabled)
          (Collision|SetCollisionEnabled (Variables|Default|GetClockwiseTrigger) "QueryOnly"))
        (else
          (CallFunction|EnableDevicePhysics (Variables|UserOptions|GetbTriggerOnProximity)))))
    (:then_1
      (Variables|Default|SetbIsDisabled false))))

(fn SetUpFlipperVectorCachedPerPlayer
  (params (PlayerToCheck :type Player%20Pawn%20Athena%20Object%20Reference%20%28by%20ref%29))
  (bind _mapflipperstaticmeshrightvectorperplayer (Variables|Default|GetMapFlipperStaticMeshRightVectorPerPlayer))
  ;[1 1 1] We want to save the right vector value the first time the player hit the flipper mesh
  (if (Utilities|Map|IsEmpty _mapflipperstaticmeshrightvectorperplayer)
    (Utilities|Map|Add _mapflipperstaticmeshrightvectorperplayer PlayerToCheck (Variables|Default|GetFlipperStaticMeshInitialRightVectorCached))
    (elif (Utilities|Map|Contains _mapflipperstaticmeshrightvectorperplayer PlayerToCheck)
      _
      (else
        (Utilities|Map|Add _mapflipperstaticmeshrightvectorperplayer PlayerToCheck (Transformation|GetRightVector (Variables|Default|GetFlipperStaticMesh)))))))

(fn SetUpFlipperVectorInitialCached
  (Variables|Default|SetFlipperStaticMeshInitialRightVectorCached (Transformation|GetRightVector (Variables|Default|GetFlipperStaticMesh))))

(fn SetupReset
  (params (Last%20Trigger%20Direction :type Float%20%28double-precision%29))
  (FlipperTriggering|AttemptBackSwingAfterDelay Last%20Trigger%20Direction)
  (if (not (Utilities|Time|IsTimerActivebyHandle (Variables|Default|GetReset%20Timer)))
    (FlipperTriggering|DisableFlipper)
    (bind _returnvalue (Utilities|Time|SetTimerbyEvent (EventDispatchers|CreateEvent :!event Reset%20Flipper) (Variables|UserOptions|Basic|GetResetTime)))
    (Variables|Default|SetReset%20Timer _returnvalue)))

(fn ShouldInvisiblePropsBeAllowed
  (return false))

(fn TriggerFlipper
  (params (Use%20Large%20Volume :type Boolean) (Trigger%20Clockwise :type Boolean) (Animate :type Boolean))
  (LocalVariables|Setlocal%20Large%20Volume Use%20Large%20Volume)
  (LocalVariables|Setlocal%20Direction Trigger%20Clockwise)
  (LocalVariables|Setlocal%20Animate Animate)
  (Variables|Default|SetTrigger%20Activation%20Tag (select (LocalVariables|Getlocal%20Direction) (select (LocalVariables|Getlocal%20Large%20Volume) (Variables|Default|GetClockwiseCueLarge) (Variables|Default|GetClockwiseCue)) (select (LocalVariables|Getlocal%20Large%20Volume) (Variables|Default|GetCounterClockwiseCueLarge) (Variables|Default|GetCounterClockwiseCue))))
  (Ability|SendGameplayEventtoActor self (Variables|Default|GetTrigger%20Activation%20Tag) (Utilities|Struct|MakeGameplayEventData (Variables|Default|GetTrigger%20Activation%20Tag) (Math|Float|float*float (select (LocalVariables|Getlocal%20Direction) 1.0 -1.0) (Math|Conversions|ToFloat(Boolean) (Utilities|Enum|NotEqual(Enum) (Variables|UserOptions|Basic|GetFlipDirection) "NewEnumerator1")) (Math|Float|SelectFloat (Variables|TriggerDirection|GetTriggerRotationMultiplier) 1.0 (LocalVariables|Getlocal%20Large%20Volume)) (Math|Conversions|ToFloat(Boolean) (not (and (or (and (Math|Boolean|ANDBoolean (== (Variables|UserOptions|Basic|GetOnBumpKnockbackPlayer) 0) (== (Variables|UserOptions|Basic|GetOnBumpKnockbackVehicle) 0) (== (Variables|UserOptions|Basic|GetOnBumpKnockbackPhysicsProp) 0)) (not (LocalVariables|Getlocal%20Large%20Volume))) (and (LocalVariables|Getlocal%20Large%20Volume) (Math|Boolean|ANDBoolean (== (Variables|UserOptions|Basic|GetOnTriggeredKnockbackPlayer) 0) (== (Variables|UserOptions|Basic|GetOnTriggeredKnockbackVehicle) 0) (== (Variables|UserOptions|Basic|GetOnTriggeredKnockbackPhysicsProp) 0)))) (Math|Boolean|ANDBoolean (== (Variables|UserOptions|Basic|GetKnockupAmountPlayer) 0) (== (Variables|UserOptions|Basic|GetKnockupAmountVehicle) 0) (== (Variables|UserOptions|Basic|GetKnockupAmountPhysicsProp) 0))))) (Math|Conversions|ToFloat(Boolean) (LocalVariables|Getlocal%20Animate))))))

(fn UpdateParentMaterial
  (params (NewParentMaterial :type Material%20Interface%20Object%20Reference))
  (CallFunction|SwitchToConstantMaterial)
  (bind _output_get (Variables|Visuals|SetParentMaterial NewParentMaterial))
  (Utilities|IsValid _output_get
    (:Is%20Valid
      (Rendering|Material|SetMaterial (Variables|Default|GetFlipperStaticMesh) 0 _output_get)
      (bind _returnvalue (Class|FortMaterialCacheManager|GetCache (Variables|Visuals|GetMaterial%20Cache%20Settings) (Variables|Visuals|GetParentMaterial)))
      (Variables|Visuals|SetMaterial%20Cache _returnvalue))))

(fn ConstructionScript)
