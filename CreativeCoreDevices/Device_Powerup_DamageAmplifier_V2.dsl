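; Creative powerup device that amplifies damage, supplying the Damage Multiplier option as effect magnitude to its parent class through overrides, along with color and FX visibility.
; InitializeShowEffects adds or removes a loose weapon-effect tag based on the Effects option, for a Gameplay Cue to read.
; BeginPlay randomizes mesh yaw and a custom primitive data float on cosmetic-capable instances.
; Saves older than release 37.20 get bAffectsDamageCap turned off.
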(event EventBeginPlay
  (|CallParentFunction)
  (Networking|CanExecuteCosmeticEvents
    (:True
      (Rendering|Material|SetCustomPrimitiveDataFloat (Variables|Components|GetStaticMeshComponent) 31 (Math|Random|RandomFloatinRange 0.0 1e+02))
      (Transformation|SetRelativeRotation (Variables|Components|GetStaticMeshComponent) (Math|Rotator|MakeRotator 0.0 0.0 (Math|Random|RandomFloatinRange 0.0 3.6e+02)) false false))))

(event Collision|EventActorBeginOverlap
  (params (OtherActor :type Actor%20Object%20Reference)))

(event EventTick
  (params (DeltaSeconds :type Float%20%28single-precision%29)))

(event HandleVersionUpgrades(CreativeVersion)
  (params (LastSavedVersion :type Creative%20Version%20Id%20Structure))
  (CreativeVersion|CompareVersiontoTag LastSavedVersion "(TagName=\"Creative.Version.Release.37.20\")"
    (:OlderVersion
      (Variables|UserOptions|SetbAffectsDamageCap false)
      (OptionData|RefreshPropertyOverrides (Variables|Components|GetToyOptionsComponent)))))

(fn GetCreativeCustomEffectData
  (return 0 0 "None" 0 0))

(fn GetCreativeEffectData
  (return "(R=0.180392,G=0.031373,B=0.282353,A=1.000000)" "NewEnumerator0" "NewEnumerator0" "NewEnumerator1"))

(fn GetEffectMagnitude
  (return (Variables|UserOptions|Basic|GetDamage%20Multiplier)))

(fn GetMagnitudeMinMax
  (return 0.0 1e+06))

(fn GetShowFXOnPlayer
  (return (Utilities|Enum|NotEqual(Enum) (Variables|UserOptions|GetEffects) "NewEnumerator0")))

(fn InitializeShowEffects
  ;[1 0.6099 0 0.2] Applies a tag to the device that will be read by the Gameplay Cue to determine if it should show damage effects or not.
  (|CallParentFunction)
  (if (Utilities|Enum|Equal(Enum) (Variables|UserOptions|GetEffects) "NewEnumerator2")
    (if (not (GameplayTags|HasMatchingGameplayTag self (Variables|Default|GetWeapon%20Effect%20Tag)))
      (Ability|GameplayEffect|AddLooseGameplayTags self (GameplayTags|MakeGameplayTagContainerfromTag (Variables|Default|GetWeapon%20Effect%20Tag))))
    (else
      (Ability|GameplayEffect|RemoveLooseGameplayTags self (GameplayTags|MakeGameplayTagContainerfromTag (Variables|Default|GetWeapon%20Effect%20Tag))))))

(fn ConstructionScript
  (|CallParentFunction))
