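; Creative powerup device that grants a colored visual effect, picking its gameplay cue from desktop, mobile or custom-effect maps.
; Device color is either custom or a dynamic team color, refreshed on team changes.
; A splitscreen visibility component gives each local player a mesh, tinted via cached dynamic materials and particle parameters, with per-player in-play visibility.
; Minigame events manage those components and join-in-progress bindings; legacy color and visibility options are migrated.
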(event EventUpdateForPropertyChange
  ;[0 1 0.1197 0.5] Binding and unbinding events to update color when team colors or player team changes.
  (|CallParentFunction)
  (CallFunction|UpdateColor)
  (Networking|CanExecuteCosmeticEvents
    (:True
      (CallFunction|BindOrUnbindTeamColorUpdates (Utilities|Enum|NotEqual(Enum) (Variables|UserOptions|Basic|GetColorType) "NewEnumerator1")))))

(event EventDestroyed
  ;[1 0.06044 0] Unbind events when this actor is removed.
  (CallFunction|BindOrUnbindTeamColorUpdates false))

(event MinigameStartedDelegate(FortMinigameProgress)
  ;[0.9219 0.07202 0.6041 0.5] Match Starting/Ending Logic
  (Default|CreateAllIndividualComponents (Variables|Default|GetCreative_SplitscreenVisibility_Component) (Variables|Components|GetStaticMeshComponent))
  (CallFunction|UpdateinPlayVisibility)
  (CallFunction|UpdateColor)
  (CallFunction|BindPlayerJoined true))

(event MinigameEndedDelegate(FortMinigameProgress)
  ;[0.9219 0.07202 0.6041 0.5] Match Starting/Ending Logic
  (Class|CreativeSplitscreenVisibilityComponent|RemoveAllIndividualComponents (Variables|Default|GetCreative_SplitscreenVisibility_Component) (Variables|Components|GetStaticMeshComponent))
  (CallFunction|UpdateColor)
  (CallFunction|BindPlayerJoined false))

(event Custom|Player%20Join%20In%20Progress
  (params (Pawn :type Pawn%20Object%20Reference))
  ;[0.599 0.5032 0.1025 0.5] Join in Progress Handling
  (CallFunction|JoininProgressLocal Pawn))

(event Custom|Join%20In%20Progress%20Local
  (params (Pawn :type Pawn%20Object%20Reference))
  (bind _returnvalue (Pawn|GetController Pawn))
  (bind _creative_splitscreenvisibility_component (Variables|Default|GetCreative_SplitscreenVisibility_Component))
  (bind _staticmeshcomponent (Variables|Components|GetStaticMeshComponent))
  ;[0.599 0.5032 0.1025 0.5] Join in Progress Handling
  (|CallParentFunction Pawn)
  (Networking|CanExecuteCosmeticEvents
    (:True
      (if (Pawn|IsLocallyControlled Pawn)
        (Class|CreativeSplitscreenVisibilityComponent|CreateIndividuallyVisibleComponent _creative_splitscreenvisibility_component _returnvalue _staticmeshcomponent)
        (Class|CreativeSplitscreenVisibilityComponent|HideOtherIndividualComponentsfromPlayer _creative_splitscreenvisibility_component _returnvalue _staticmeshcomponent)
        (CallFunction|UpdateinPlayVisibility)))))

(event Custom|Player%20Removed
  (params (UniqueNetId :type Unique%20Net%20Id%20Structure) (bIsLocalPlayer :type Boolean))
  (bind _returnvalue (Game|GetPlayerStatefromUniqueNetId UniqueNetId))
  (bind _returnvalue_2 (PlayerState|GetPlayerController _returnvalue))
  (bind _creative_splitscreenvisibility_component (Variables|Default|GetCreative_SplitscreenVisibility_Component))
  ;[0.599 0.5032 0.1025 0.5] Join in Progress Handling
  (Networking|CanExecuteCosmeticEvents
    (:True
      (if bIsLocalPlayer
        (Class|CreativeSplitscreenVisibilityComponent|RemoveIndividuallyVisibleComponent _creative_splitscreenvisibility_component _returnvalue_2)
        (Class|CreativeSplitscreenVisibilityComponent|RemovePlayerfromAllComponents _creative_splitscreenvisibility_component _returnvalue_2)))))

(fn BindorUnbindTeamColorUpdates
  (params (Bind :type Boolean))
  (bind _outputdelegate (EventDispatchers|CreateEvent :!event Update%20Color))
  (bind _fortminigameprogress (Variables|Default|GetFortMinigameProgress))
  ;[1 1 1] We need to listen for when either the team colors change or when the player changes teams. These combined cover the options of when we want to update the color of the device.
  ;
  (if Bind
    (Teams|RefreshTeamChangeBinding _fortminigameprogress false _outputdelegate)
    (bind _array_element (Default|ForEachFortLocalPlayerController)
      (:Loop%20Body
        (bind _asathena_player_controller (Utilities|Casting|CastToAthena_PlayerController _array_element))
        (FortPlayerControllerAthena|BindEventtoOnLocalPlayerChangedTeams _asathena_player_controller _outputdelegate)))
    (else
      (Teams|RefreshTeamChangeBinding _fortminigameprogress true _outputdelegate)
      (bind _array_element_2 (Default|ForEachFortLocalPlayerController)
        (:Loop%20Body
          (bind _asathena_player_controller_2 (Utilities|Casting|CastToAthena_PlayerController _array_element_2))
          (FortPlayerControllerAthena|UnbindEventfromOnLocalPlayerChangedTeams _asathena_player_controller_2 _outputdelegate))))))

(fn BindPlayerJoined
  (params (Bind :type Boolean))
  (bind _outputdelegate (EventDispatchers|CreateEvent :!event Player%20Removed))
  (bind _outputdelegate_2 (EventDispatchers|CreateEvent :!event Player%20Join%20In%20Progress))
  (bind _fortminigameprogress (Variables|Default|GetFortMinigameProgress))
  (bind _currentminigame (Variables|/Script/FortniteGame.FortMinigameLogicComponent|GetCurrentMinigame _fortminigameprogress))
  (bind _fortminigameprogress_2 (Variables|Default|GetFortMinigameProgress))
  (bind _currentminigame_2 (Variables|/Script/FortniteGame.FortMinigameLogicComponent|GetCurrentMinigame _fortminigameprogress_2))
  (Utilities|IsValid (Variables|/Script/FortniteGame.FortMinigameLogicComponent|GetCurrentMinigame (Variables|Default|GetFortMinigameProgress))
    (:Is%20Valid
      (if (Variables|GetBind)
        (Minigame|BindEventtoOnPlayerJoininProgress _currentminigame_2 _outputdelegate_2)
        (Minigame|BindEventtoPlayerRemovedDelegate _currentminigame_2 _outputdelegate)
        (else
          (Minigame|UnbindEventfromOnPlayerJoininProgress _currentminigame _outputdelegate_2)
          (Minigame|UnbindEventfromPlayerRemovedDelegate _currentminigame _outputdelegate))))))

(fn CreateorUpdateMeshMID
  (params (Mesh%20Component :type Mesh%20Component%20Object%20Reference))
  (bind _mesh_mid (Variables|Default|GetMesh%20MID))
  (bind (_value _returnvalue) (Utilities|Map|Find _mesh_mid Mesh%20Component))
  (if _returnvalue
    ;[1 1 1] If MID Exists but isn't attached, attach it. (Move Tool fix. This isn't a BuildingSMActor, so can't use Get Override Mesh Material
    (if (Utilities|NotEqual(Object) _value (Rendering|Material|GetMaterial (Variables|GetMesh%20Component) 0))
      (Rendering|Material|SetMaterial (Variables|GetMesh%20Component) 0 _value))
    (else
      ;[1 1 1] If MID isn't already created, create it and save a ref
      (Utilities|Map|Add (Variables|Default|GetMesh%20MID) (Variables|GetMesh%20Component) (Material|GetMID (Variables|GetMesh%20Component))))))

(fn GetCreativeCustomEffectData
  (bind _value (Utilities|Map|Find (Variables|Default|GetSocketDetails) (Variables|UserOptions|Basic|GetCustom%20Effect%20Socket)))
  (bind (_value_name_5_2a3abbc64d7c11c5471932a7374b8f02 _value_offset_11_a08bc9e5427de43e722806a0a1da5a53) (Utilities|Struct|BreakVFXSocketDetails _value))
  (return (Variables|UserOptions|Basic|GetCustom%20Effect) (Variables|UserOptions|Basic|GetCustom%20Audio) _value_name_5_2a3abbc64d7c11c5471932a7374b8f02 _value_offset_11_a08bc9e5427de43e722806a0a1da5a53 (Variables|UserOptions|Basic|GetCustom%20Effect%20Exclusivity%20Index)))

(fn GetCreativeEffectData
  (return (CallFunction|GetCreativeEffectLinear)))

(fn GetDeviceColor
  (params (Player%20Controller :type Player%20Controller%20Object%20Reference))
  (bind _player_controller (Variables|GetPlayer%20Controller))
  (switch Utilities|FlowControl|Switch|SwitchonECreativeColorSetType (Variables|UserOptions|Basic|GetColorType)
    (:NewEnumerator1
      ;[1 1 1] Not using Team Color. Pass back the color index from the user option.
      (return (Variables|UserOptions|Basic|GetCustomColor)))
    (:NewEnumerator2
      _)
    (:NewEnumerator3
      _))
  ;[1 1 1] Override Color Index with Dynamic color (Team/TeamRelationship) if that's what we're looking for.
  (Utilities|IsValid _player_controller
    (:Is%20Valid
      (bind _playerstate (Variables|/Script/Engine.Controller|GetPlayerState _player_controller)
        (:then
          (bind _fortminigameprogress (Variables|Default|GetFortMinigameProgress)
            (:then
              (bind _currentminigame (Variables|/Script/FortniteGame.FortMinigameLogicComponent|GetCurrentMinigame _fortminigameprogress)
                (:then
                  (bind _output_get (LocalVariables|SetLocal%20Color%20Index (Teams|GetDynamicTeamColor _playerstate _currentminigame (Variables|UserOptions|Basic|GetColorType))))
                  (return (Utilities|Array|Get(acopy) (Class|GetClassDefaults "/Game/Creative/Sets/MilitaryBase/Props/CreativeTeamColors.CreativeTeamColors_C") (Math|Integer|%25%28Integer%29 _output_get (Utilities|Array|Length (Class|GetClassDefaults "/Game/Creative/Sets/MilitaryBase/Props/CreativeTeamColors.CreativeTeamColors_C"))))))
                (:else
                  _)))
            (:else
              _))
          (return "(R=1.000000,G=1.000000,B=1.000000,A=1.000000)"))
        (:else
          _)))
    (:Is%20Not%20Valid
      _)))

(fn GetGameplayCue
  (bind _returnvalue (TouchPlatform|IsMobileGame))
  (bind _visualeffect (Variables|UserOptions|Basic|GetVisualEffect))
  (bind _mobilegameplayeffectmap (Variables|Default|GetMobileGameplayEffectMap))
  (bind _gameplayeffectmap (Variables|Default|GetGameplayEffectMap))
  (bind _returnvalue_2 (select _returnvalue _mobilegameplayeffectmap _gameplayeffectmap))
  (bind (_value _returnvalue_3) (Utilities|Map|Find _returnvalue_2 _visualeffect))
  (if (Utilities|Enum|Equal(Enum) (Variables|UserOptions|Basic|GetVisualEffect) "NewEnumerator5")
    (bind _result (ArrayMacros|SafeArrayGet (select (Variables|UserOptions|GetPersistOnElimination) (Variables|Default|GetCustomEffectArray_PersistOnDeath) (Variables|Default|GetCustomEffectArray)) (Variables|UserOptions|Basic|GetCustom%20Effect%20Exclusivity%20Index))
      (:%20%20
        (return _result)))
    (elif _returnvalue_3
      (return _value)
      (else
        (return 0)))))

(fn GetMeshComponent
  ;[1 1 1] Overriding Mesh Component with a new one. This doesn't use the base component as that doesn't interact with scale and the move tool properly
  ;
  (return 0))

(fn GetCreativeEffectLinear
  (return (Utilities|Struct|MakeFCreativeEffectDirectColor (CallFunction|GetDeviceColor (Fortnite|GetFirstLocalFortPlayerController true)) (Variables|UserOptions|Basic|GetColorType) (select (Variables|UserOptions|GetbIsEffectVisibleToLocalPlayer) "NewEnumerator0" "NewEnumerator1") (Variables|UserOptions|GetTeamRelationshipVisibility))))

(fn MigrateLegacyUserOptions
  (bind _knob_colorindex (Class|GetClassDefaults "/CreativeCoreDevices/Device_Powerup_VisualEffect_V2.Device_Powerup_VisualEffect_V2_C"))
  (bind _knob_isvisibletoowningplayer (Class|GetClassDefaults "/CreativeCoreDevices/Device_Powerup_VisualEffect_V2.Device_Powerup_VisualEffect_V2_C"))
  (|CallParentFunction)
  (Utilities|FlowControl|Sequence
    (:then_0
      (Utilities|FlowControl|Sequence
        (:then_0
          (if (Utilities|Enum|NotEqual(Enum) (Variables|Legacy|GetKnob_IsVisibleToOwningPlayer) _knob_isvisibletoowningplayer)
            (Variables|UserOptions|SetbIsEffectVisibleToLocalPlayer false)
            (Variables|Legacy|SetKnob_IsVisibleToOwningPlayer _knob_isvisibletoowningplayer)
            (LocalVariables|SetbMigratedProperties true)))
        (:then_1
          (if (!= (Variables|Legacy|GetKnob_ColorIndex) _knob_colorindex)
            (bind _output_get (Variables|UserOptions|Basic|SetColorType (select (>= (Variables|Legacy|GetKnob_ColorIndex) 0) "NewEnumerator1" (select (== (Variables|Legacy|GetKnob_ColorIndex) -1) "NewEnumerator2" "NewEnumerator3"))))
            (if (and (Utilities|Enum|Equal(Enum) _output_get "NewEnumerator1") (Utilities|Array|IsValidIndex (Variables|Legacy|GetIndexToColor) (Variables|Legacy|GetKnob_ColorIndex)))
              (Variables|UserOptions|Basic|SetCustomColor (Utilities|Array|Get(acopy) (Variables|Legacy|GetIndexToColor) (Variables|Legacy|GetKnob_ColorIndex)))
              (else
                _))
            (Variables|Legacy|SetKnob_ColorIndex _knob_colorindex)
            (LocalVariables|SetbMigratedProperties true)))))
    (:then_1
      (if (LocalVariables|GetbMigratedProperties)
        (OptionData|RefreshPropertyOverrides (Variables|Components|GetToyOptionsComponent))))))

(fn UpdateColor
  (Networking|CanExecuteCosmeticEvents
    (:True
      (CallFunction|UpdatePowerupColors))))

(fn UpdatePowerupColors
  (bind _localmesh (LocalVariables|GetLocalMesh))
  (bind _mesh_mid (Variables|Default|GetMesh%20MID))
  (bind _value (Utilities|Map|Find _mesh_mid _localmesh))
  (bind _localmesh_2 (LocalVariables|GetLocalMesh))
  (bind _array_element (Default|ForEachFortLocalPlayerController)
    (:Loop%20Body
      (LocalVariables|SetColor (CallFunction|GetDeviceColor _array_element))
      (bind (_scene_component _found) (Class|CreativeSplitscreenVisibilityComponent|GetUniqueComponentforPlayer (Variables|Default|GetCreative_SplitscreenVisibility_Component) _array_element))
      (if _found
        (bind _asmesh_component (Utilities|Casting|CastToMeshComponent _scene_component))
        (LocalVariables|SetLocalMesh _asmesh_component)
        (else
          (LocalVariables|SetLocalMesh (Variables|Components|GetStaticMeshComponent))))
      (CallFunction|CreateOrUpdateMeshMID _localmesh_2)
      (for _array_element_2 (Components|GetChildrenComponents _localmesh_2 false)
        (bind _ascascade_particle_system_component (Utilities|Casting|CastToParticleSystemComponent _array_element_2))
        (LocalVariables|SetLocalFX _ascascade_particle_system_component))
      (Components|Activation|Deactivate (LocalVariables|GetLocalFX))
      (Effects|Components|ParticleSystem|SetColorParameter (LocalVariables|GetLocalFX) "Color" (LocalVariables|GetColor))
      (Rendering|Material|SetVectorParameterValue _value "ColorHigh" (LocalVariables|GetColor) :!class /Script/Engine.MaterialInstanceDynamic)
      (Rendering|Material|SetVectorParameterValue _value "Color_2" (Math|Color|LinearColor*Float (LocalVariables|GetColor) (Variables|Default|GetMesh%20Material%20Multiplier)) :!class /Script/Engine.MaterialInstanceDynamic)
      (Variables|Default|SetVFX%20Particle%20Color (LocalVariables|GetColor))
      (Components|Activation|Activate (LocalVariables|GetLocalFX) true) ; Light Color of the FX doesn't update directly when you change it.\nThis forces the FX to reset.
    )))

(fn UpdateInPlayVisibility
  (switch Utilities|FlowControl|Switch|SwitchonEnum_Creative_Powerup_Visibility (Variables|UserOptions|GetWhoCanSeeThisPowerup)
    (:NewEnumerator0
      (Rendering|SetActorHiddenInGame true))
    (:NewEnumerator1
      (Rendering|SetActorHiddenInGame false))
    (:NewEnumerator2
      (Networking|CanExecuteCosmeticEvents
        (:True
          (bind _array_element (Default|ForEachFortLocalPlayerController)
            (:Loop%20Body
              (bind (_scene_component _found) (Class|CreativeSplitscreenVisibilityComponent|GetUniqueComponentforPlayer (Variables|Default|GetCreative_SplitscreenVisibility_Component) _array_element))
              (if _found
                (Class|CreativeSplitscreenVisibilityComponent|HideSceneComponentfromPlayer (Variables|Default|GetCreative_SplitscreenVisibility_Component) _array_element _scene_component (not (Class|CreativeUseRestrictionComponent|StateChecks (Variables|Components|GetCreativeUseRestriction) _array_element true)))))
            (:Completed)))))))

(fn ConstructionScript
  (|CallParentFunction)
  (CallFunction|UpdateColor))
