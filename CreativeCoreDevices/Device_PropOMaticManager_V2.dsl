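; Creative Prop-O-Matic Manager device that configures prop-hunt disguise, ping, HUD and item drop rules for a minigame.
; On dedicated-server begin play and minigame start, it registers itself as the minigame's PropHuntSettings and sets the drop rule.
; Verse/event functions ping props and set replicated overrides that supersede user options, with some RepNotifies refreshing settings; overrides reset at minigame end.
; It also broadcasts disguise and ping events and migrates legacy options.
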(event EventOnWorldReady
  ;[0 1 0.08672] Visible In Game Component needs initialized even when working in basic mode as BP components can't initialize in editor,
  ; and BuildingProp based devices need to initialize visibleingame OnWorldReady
  (|CallParentFunction)
  (Class|CreativeVisibleinGameComponent|InitializeComponent (Variables|Default|GetVisibilityRulesComponent)))

(event EditorPreview|EventSetupForEditorPreview
  ;[0 1 0.08672] Visible In Game Component needs initialized even when working in basic mode as BP components can't initialize in editor,
  ; and BuildingProp based devices need to initialize visibleingame OnWorldReady
  (|CallParentFunction)
  (Class|CreativeVisibleinGameComponent|InitializeComponent (Variables|Default|GetVisibilityRulesComponent)))

(event EventBeginPlay
  (Networking|CanExecuteCosmeticEvents
    (:False
      (bind _asminigame_player_created (Utilities|Casting|CastToMinigame_PlayerCreated (Minigame|GetMinigame self :!class /Script/FortniteGame.MinigameBlueprintLibrary)))
      (Variables|/Game/Athena/Playset/Minigames/Minigame_PlayerCreated.Minigame_PlayerCreated_C|SetPropHuntSettings self _asminigame_player_created)
      (CallFunction|SetMinigameDropRule _asminigame_player_created))))

(event MinigameStartedDelegate(FortMinigameProgress)
  (Networking|CanExecuteCosmeticEvents
    (:False
      (bind _asminigame_player_created (Utilities|Casting|CastToMinigame_PlayerCreated (Minigame|GetMinigame self :!class /Script/FortniteGame.MinigameBlueprintLibrary)))
      (Variables|/Game/Athena/Playset/Minigames/Minigame_PlayerCreated.Minigame_PlayerCreated_C|SetPropHuntSettings self _asminigame_player_created)
      (CallFunction|SetMinigameDropRule _asminigame_player_created))))

(event MinigameEndedDelegate(FortMinigameProgress)
  (CallFunction|ResetVerseOverrides))

(event CreativePropPossession|EventOnPlayerBeginEnteringDisguise
  (params (Instigator :type Controller%20Object%20Reference))
  (GameplayEvent|BroadcastEvent (Variables|UserOptions|Events|GetOn%20Begin%20Entering%20Disguise) 0 (Utilities|Struct|MakeGameplayEventonBeginEnteringDisguise Instigator)))

(event CreativePropPossession|EventOnPlayerFinishEnteringDisguise
  (params (Instigator :type Controller%20Object%20Reference))
  (GameplayEvent|BroadcastEvent (Variables|UserOptions|Events|GetOn%20Finish%20Entering%20Disguise) 0 (Utilities|Struct|MakeGameplayEventonFinishEnteringDisguise Instigator)))

(event CreativePropPossession|EventOnPlayerExitingDisguise
  (params (Instigator :type Controller%20Object%20Reference))
  (GameplayEvent|BroadcastEvent (Variables|UserOptions|Events|GetOn%20Exiting%20Disguise) 0 (Utilities|Struct|MakeGameplayEventonExitingDisguise Instigator)))

(event Custom|Ping%20Player%20Props%20Event
  (params (Instigator :type Controller%20Object%20Reference))
  ;[1 1 1] Gameplay Functions
  (Verse|PingPlayerProps))

(event Custom|On%20All%20Props%20Pinged%20Event
  (params (Instigator :type Controller%20Object%20Reference))
  ;[1 1 1] Gameplay Functions
  (GameplayEvent|BroadcastEvent (Variables|UserOptions|Events|GetOn%20All%20Props%20Pinged) 0 (Utilities|Struct|MakeGameplayEventonAllPropsPinged Instigator)))

(event Custom|Ping%20Player%20Prop%20Event
  (params (Instigator :type Controller%20Object%20Reference))
  ;[1 1 1] Gameplay Functions
  (Verse|PingPlayerProp Instigator))

(event Custom|On%20Prop%20Pinged%20Event
  (params (Instigator :type Controller%20Object%20Reference))
  ;[1 1 1] Gameplay Functions
  (GameplayEvent|BroadcastEvent (Variables|UserOptions|Events|GetOn%20Prop%20Pinged) 0 (Utilities|Struct|MakeGameplayEventonPropPinged Instigator)))

(event Custom|Toggle%20Ping%20Props%20Event
  (params (Instigator :type Controller%20Object%20Reference))
  ;[1 1 1] Gameplay Functions
  (Verse|SetPingProps (not (Utilities|Select :Index (Variables|Verse|GetPingHiddenPropsOnIntervalOverride) :NewEnumerator0 (Variables|UserOptions|Basic|GetPing%20Hidden%20Props%20on%20Interval) :NewEnumerator1 true :NewEnumerator2 false))))

(event Custom|Toggle%20Show%20Props%20Remaining%20Event
  (params (Instigator :type Controller%20Object%20Reference))
  ;[1 1 1] Gameplay Functions
  (Verse|SetShowPropsRemaining (not (Utilities|Select :Index (Variables|Verse|GetShouldShowPropsRemainingOverride) :NewEnumerator0 (Variables|UserOptions|Basic|GetShouldShowPropsRemaining) :NewEnumerator1 true :NewEnumerator2 false))))

(event Custom|Toggle%20Show%20Prop%20Ping%20Cooldown%20Event
  (params (Instigator :type Controller%20Object%20Reference))
  ;[1 1 1] Gameplay Functions
  (Verse|SetShowPropPingCooldown (not (Utilities|Select :Index (Variables|Verse|GetShowPropPingCooldownOverride) :NewEnumerator0 (Variables|UserOptions|Advanced|GetShowPropPingCooldown) :NewEnumerator1 true :NewEnumerator2 false))))

(fn EquipPrimaryItemAfterCancelPossession
  (return (Variables|UserOptions|Basic|GetEquipPickaxeAfterCancellingDisguise)))

(fn GetDisguiseAnimationOverrideDetails
  (return (Variables|UserOptions|Basic|GetDisguiseAnimationDuration_Override) (Variables|UserOptions|Basic|GetDisguiseAnimationDuration)))

(fn GetPingDetails
  (return (Utilities|Select :Index (Variables|Verse|GetShowPropPingCooldownOverride) :NewEnumerator0 (Variables|UserOptions|Advanced|GetShowPropPingCooldown) :NewEnumerator1 true :NewEnumerator2 false) (Math|Float|SelectFloat (Math|Float|SelectFloat (Variables|UserOptions|Basic|GetPropPingFrequency) (Variables|Verse|GetPropPingFrequencyOverride) (< (Variables|Verse|GetPropPingFrequencyOverride) 0)) 0.0 (Utilities|Select :Index (Variables|Verse|GetPingHiddenPropsOnIntervalOverride) :NewEnumerator0 (Variables|UserOptions|Basic|GetPing%20Hidden%20Props%20on%20Interval) :NewEnumerator1 true :NewEnumerator2 false))))

(fn GetPropDetails
  (return (Utilities|Select :Index (Variables|Verse|GetShouldShowPropsRemainingOverride) :NewEnumerator0 (Variables|UserOptions|Basic|GetShouldShowPropsRemaining) :NewEnumerator1 true :NewEnumerator2 false) (Utilities|Enum|Equal(Enum) (Variables|UserOptions|Basic|GetPropHealthBehavior) "NewEnumerator1") (Variables|UserOptions|Basic|GetAllowChangingDisguises) (Variables|UserOptions|Basic|GetAllowDisguiseToBeCanceled)))

(fn GetFilteringDetails
  (return (Variables|UserOptions|Advanced|GetDisallowed%20Prop%20List)))

(fn HasUEFNOnlyOptions
  (return (Utilities|Array|IsNotEmpty (Variables|UserOptions|Advanced|GetDisallowed%20Prop%20List)) (Utilities|Text|MakeLiteralText "NSLOCTEXT(\"[573DE0E64E13FA508E66BD8206B8E62B]\", \"F9088C1849A4900E53ACE4936553AB6B\", \"This device contains props in its Disallowed Prop List,\\r\\nwhich can only be edited in UEFN.\")")))

(fn IsPlayerProp
  (params (Controller :type Controller%20Object%20Reference))
  (Utilities|IsValid Controller
    (:Is%20Valid
      (for _array_element (Actor|GetAttachedActors (Pawn|GetControlledPawn Controller) false false)
        (Utilities|Casting|CastToCreative_Player_Prop _array_element)
        (return true)))
    (:Is%20Not%20Valid
      _))
  (return false))

(fn MigrateLegacyUserOptions
  (bind _showpropsremaining (Class|GetClassDefaults "/CreativeCoreDevices/Device_PropOMaticManager_V2.Device_PropOMaticManager_V2_C"))
  (bind _allowcancelproppossession (Class|GetClassDefaults "/CreativeCoreDevices/Device_PropOMaticManager_V2.Device_PropOMaticManager_V2_C"))
  (bind _allowchangepossessedprop (Class|GetClassDefaults "/CreativeCoreDevices/Device_PropOMaticManager_V2.Device_PropOMaticManager_V2_C"))
  (bind _prophealthsetting (Class|GetClassDefaults "/CreativeCoreDevices/Device_PropOMaticManager_V2.Device_PropOMaticManager_V2_C"))
  (bind _equipprimaryitemaftercancelproppossession (Class|GetClassDefaults "/CreativeCoreDevices/Device_PropOMaticManager_V2.Device_PropOMaticManager_V2_C"))
  (bind _pingfrequency (Class|GetClassDefaults "/CreativeCoreDevices/Device_PropOMaticManager_V2.Device_PropOMaticManager_V2_C"))
  (bind _pingfrequency_2 (Variables|LegacyOptions|GetPingFrequency))
  (|CallParentFunction)
  (Utilities|FlowControl|Sequence
    (:then_0
      (Utilities|FlowControl|Sequence
        (:then_0
          ;[1 0.2417 0.9092] Converting Show Props remaining int->bool
          (if (!= (Variables|LegacyOptions|GetShowPropsRemaining) _showpropsremaining)
            (Variables|UserOptions|Basic|SetShouldShowPropsRemaining false)
            (Variables|LegacyOptions|SetShowPropsRemaining _showpropsremaining)
            (Variables|Default|SetMigrated%20Properties true)))
        (:then_1
          ;[0.2661 0.3223 1] Converting Allow Cancel Prop Possession int->bool
          (if (!= (Variables|LegacyOptions|GetAllowCancelPropPossession) _allowcancelproppossession)
            (Variables|UserOptions|Basic|SetAllowDisguiseToBeCanceled false)
            (Variables|LegacyOptions|SetAllowCancelPropPossession _allowcancelproppossession)
            (Variables|Default|SetMigrated%20Properties true)))
        (:then_2
          ;[0.2882 1 0.2592] Converting AllowChangePossessedProp int->bool
          (if (!= (Variables|LegacyOptions|GetAllowChangePossessedProp) _allowchangepossessedprop)
            (Variables|UserOptions|Basic|SetAllowChangingDisguises false)
            (Variables|LegacyOptions|SetAllowChangePossessedProp _allowchangepossessedprop)
            (Variables|Default|SetMigrated%20Properties true)))
        (:then_3
          ;[0.9732 1 0.2218] Converting PropHealthSetting int->Enum
          (if (!= (Variables|LegacyOptions|GetPropHealthSetting) _prophealthsetting)
            (Variables|UserOptions|Basic|SetPropHealthBehavior "NewEnumerator0")
            (Variables|LegacyOptions|SetPropHealthSetting _prophealthsetting)
            (Variables|Default|SetMigrated%20Properties true)))
        (:then_4
          ;[1 0.1696 0.08993] Converting DropGunBehavior int->enum
          (if (!= (Variables|LegacyOptions|GetDropPropGunBehavior) (Class|GetClassDefaults "/CreativeCoreDevices/Device_PropOMaticManager_V2.Device_PropOMaticManager_V2_C"))
            (Variables|UserOptions|Advanced|SetDropProp-O-MaticBehavior_Override true)
            (Variables|UserOptions|Advanced|SetDropProp-O-MaticBehavior (Utilities|Select :Index (Variables|LegacyOptions|GetDropPropGunBehavior) :Option%200 "NewEnumerator0" :Option%201 "NewEnumerator0" :Option%202 "NewEnumerator1" :Option%203 "NewEnumerator2"))
            (Variables|LegacyOptions|SetDropPropGunBehavior (Class|GetClassDefaults "/CreativeCoreDevices/Device_PropOMaticManager_V2.Device_PropOMaticManager_V2_C"))
            (Variables|Default|SetMigrated%20Properties true)))
        (:then_5
          ;[0.3732 0.08683 1] Converting EquipPrimaryItemAfterCancelPropPossession int->Bool
          (if (!= (Variables|LegacyOptions|GetEquipPrimaryItemAfterCancelPropPossession) _equipprimaryitemaftercancelproppossession)
            (Variables|UserOptions|Basic|SetEquipPickaxeAfterCancellingDisguise true)
            (Variables|LegacyOptions|SetEquipPrimaryItemAfterCancelPropPossession _equipprimaryitemaftercancelproppossession)
            (Variables|Default|SetMigrated%20Properties true)))
        (:then_6
          ;[0.9404 0.3474 1] Converting/Split PingFrequency int -> Bool + Float
          (if (!= _pingfrequency_2 _pingfrequency)
            (Variables|UserOptions|Basic|SetPing%20Hidden%20Props%20on%20Interval (!= _pingfrequency_2 0))
            (Variables|UserOptions|Basic|SetPropPingFrequency (Math|Conversions|ToFloat(Integer) _pingfrequency_2))
            (Variables|LegacyOptions|SetPingFrequency _pingfrequency)
            (Variables|Default|SetMigrated%20Properties true)))))
    (:then_1
      (if (Variables|Default|GetMigrated%20Properties)
        (OptionData|RefreshPropertyOverrides (Variables|Default|GetToyOptionsComponent))))))

(fn OnRep_PingHiddenPropsOnIntervalOverride
  (bind _asminigame_player_created (Utilities|Casting|CastToMinigame_PlayerCreated (Minigame|GetMinigame self :!class /Script/FortniteGame.MinigameBlueprintLibrary)))
  (Class|MinigamePlayerCreated|UpdatePropHuntSettings _asminigame_player_created))

(fn OnRep_PropPingFrequencyOverride
  (bind _asminigame_player_created (Utilities|Casting|CastToMinigame_PlayerCreated (Minigame|GetMinigame self :!class /Script/FortniteGame.MinigameBlueprintLibrary)))
  (Class|MinigamePlayerCreated|UpdatePropHuntSettings _asminigame_player_created))

(fn OnRep_ShouldShowPropsRemainingOverride
  (bind _asminigame_player_created (Utilities|Casting|CastToMinigame_PlayerCreated (Minigame|GetMinigame self :!class /Script/FortniteGame.MinigameBlueprintLibrary)))
  (Class|MinigamePlayerCreated|UpdatePropHuntSettings _asminigame_player_created))

(fn PingPlayerProp
  (params (Controller :type Controller%20Object%20Reference))
  (Utilities|IsValid Controller
    (:Is%20Valid
      (for _array_element (Actor|GetAttachedActors (Pawn|GetControlledPawn Controller) false false)
        (bind _ascreative_player_prop (Utilities|Casting|CastToCreative_Player_Prop _array_element))
        (Class|CreativePlayerProp|PlayHintSoundonClient _ascreative_player_prop)
        (return))
      (CallFunction|OnPropPingedEvent Controller))))

(fn PingPlayerProps
  (bind _outplayers (Minigame|GetParticipatingPlayers_NotPure (Minigame|GetMinigame self :!class /Script/FortniteGame.MinigameBlueprintLibrary)))
  (for _array_element _outplayers
    (bind _array_element_2 (Utilities|Array|ForEachLoopwithBreak (Actor|GetAttachedActors (PlayerState|GetPawn _array_element) false false))
      (:LoopBody
        (bind _ascreative_player_prop (Utilities|Casting|CastToCreative_Player_Prop _array_element_2))
        (Class|CreativePlayerProp|PlayHintSoundonClient _ascreative_player_prop))))
  (CallFunction|OnAllPropsPingedEvent 0))

(fn ResetVerseOverrides
  (Variables|Verse|SetPingHiddenPropsOnIntervalOverride "NewEnumerator0")
  (Variables|Verse|SetPropPingFrequencyOverride -1.0)
  (Variables|Verse|SetShouldShowPropsRemainingOverride "NewEnumerator0")
  (Variables|Verse|SetShowPropPingCooldownOverride "NewEnumerator0"))

(fn SetMinigameDropRule
  (params (Minigame :type Minigame%20Player%20Created%20Object%20Reference))
  (bind _proppossessoritemdef (Variables|Default|GetPropPossessorItemDef))
  (bind _minigame (Variables|GetMinigame))
  (bind _itemdropoverride (Variables|/Game/Athena/Playset/Minigames/Minigame_PlayerCreated.Minigame_PlayerCreated_C|GetItemDropOverride _minigame))
  (bind (_value _returnvalue) (Utilities|Map|Find _itemdropoverride _proppossessoritemdef))
  (Utilities|IsValid Minigame
    (:Is%20Valid
      (if _returnvalue
        (if (Variables|UserOptions|Advanced|GetDropProp-O-MaticBehavior_Override)
          (Variables|SetBy-RefVar _value (Variables|UserOptions|Advanced|GetDropProp-O-MaticBehavior))
          (else
            (Utilities|Map|Remove (Variables|/Game/Athena/Playset/Minigames/Minigame_PlayerCreated.Minigame_PlayerCreated_C|GetItemDropOverride (Variables|GetMinigame)) (Variables|Default|GetPropPossessorItemDef))))
        (else
          (if (Variables|UserOptions|Advanced|GetDropProp-O-MaticBehavior_Override)
            (Utilities|Map|Add (Variables|/Game/Athena/Playset/Minigames/Minigame_PlayerCreated.Minigame_PlayerCreated_C|GetItemDropOverride (Variables|GetMinigame)) (Variables|Default|GetPropPossessorItemDef) (Variables|UserOptions|Advanced|GetDropProp-O-MaticBehavior))))))))

(fn SetPingFrequency
  (params (Time :type Float%20%28double-precision%29))
  (Variables|Verse|SetPropPingFrequencyOverride (Math|Float|Clamp(Float) Time 3.0 3e+01)))

(fn SetPingProps
  (params (On :type Boolean))
  (Variables|Verse|SetPingHiddenPropsOnIntervalOverride (select On "NewEnumerator1" "NewEnumerator2")))

(fn SetShowPropPingCooldown
  (params (On :type Boolean))
  (Variables|Verse|SetShowPropPingCooldownOverride (select On "NewEnumerator1" "NewEnumerator2")))

(fn SetShowPropsRemaining
  (params (On :type Boolean))
  (Variables|Verse|SetShouldShowPropsRemainingOverride (select On "NewEnumerator1" "NewEnumerator2")))

(fn ConstructionScript)
