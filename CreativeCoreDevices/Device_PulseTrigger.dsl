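; Creative Pulse Trigger device that sweeps a resizable pulse volume through a zone, triggering music sequencer objects, damaging targets and broadcasting player-hit events.
; Pad overlaps, chained pulses, device events or game phases start, toggle or pause it.
; A tempo-scaled timeline moves the pulse, with looping and bouncing.
; The server replicates start time, signed for direction, and pause position; OnRep handlers replay the wave in sync on clients.
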(event OnComponentBeginOverlap(PadTrigger)
  (params (OverlappedComponent :type Primitive%20Component%20Object%20Reference) (OtherActor :type Actor%20Object%20Reference) (OtherComp :type Primitive%20Component%20Object%20Reference) (OtherBodyIndex :type Integer) (bFromSweep :type Boolean) (SweepResult :type Hit%20Result%20Structure%20%28by%20ref%29))
  ;[1 1 1] Handles checking for overlap from a player hitting the pad.
  (if (and (not (Building|IsPreviewBuildingActor)) (Class|CreativeEnabledComponent|IsEnabled (Variables|Default|GetEnabledComponent)))
    (bind _validinstigator (CallFunction|ValidateAndSetInstigator OtherActor))
    (if _validinstigator
      (CallFunction|StartfromTrigger))))

(event OnComponentBeginOverlap(ChainTrigger)
  (params (OverlappedComponent :type Primitive%20Component%20Object%20Reference) (OtherActor :type Actor%20Object%20Reference) (OtherComp :type Primitive%20Component%20Object%20Reference) (OtherBodyIndex :type Integer) (bFromSweep :type Boolean) (SweepResult :type Hit%20Result%20Structure%20%28by%20ref%29))
  ;[1 1 1] Handles checking for overlap from a moving trigger of another Sweep trigger
  (if (Math|Boolean|ANDBoolean (Components|ComponentHasTag OtherComp "AffectsChainTrigger") (Utilities|NotEqual(Object) OtherComp (Variables|Default|GetMovingTrigger)) (Class|CreativeEnabledComponent|IsEnabled (Variables|Default|GetEnabledComponent)))
    (if (and (Utilities|DoesObjectImplementInterface OtherActor "/Game/Creative/Devices/Common/Creative_DeviceInterface.Creative_DeviceInterface_C") (Variables|UserOptions|GetSend%20Player%20Information))
      (bind _instigator (CreativeDevice|GetInstigatingController(Message) OtherActor))
      (Variables|Default|SetTriggerInstigator _instigator))
    (Variables|Default|SetInitialTrigger false) ; We don't want chained sweeps to have audio
    (CallFunction|StartfromTrigger)))

(event Custom|Fire%20Sequencer%20from%20Start
  ;[0 0.3366 1] Starts the Sequencer from the beginning.
  (CallFunction|SetInitialDirection)
  (Variables|Default|SetNumLoopsRun 0) ; reset loop count
  (CallFunction|ClearPauseLocation)
  (CallFunction|ResumeWave))

(event Custom|StartFromTrigger
  (if (and (or (not (Variables|UserOptions|GetDisable%20Activation%20While%20Running)) (not (Components|Timeline|IsPlaying (Variables|Device_PulseTrigger|GetTriggerMovement)))) (Class|CreativeEnabledComponent|IsEnabled (Variables|Default|GetEnabledComponent)))
    (CallFunction|PlayAudio)
    ;[1 1 1] Trigger Type Setting handling
    (switch Utilities|FlowControl|Switch|SwitchonECreativeSweepTriggerType (Variables|UserOptions|GetActivation%20Type)
      (:NewEnumerator0
        (CallFunction|FireSequencerFromStart))
      (:NewEnumerator1
        (CallFunction|ToggleLogic))
      (:NewEnumerator2
        (CallFunction|TogglePauseLogic)))))

(event Custom|On%20Wave%20Fired
  (params (Start%20Delta :type Float%20%28double-precision%29))
  (bind _numloopsrun (Variables|Default|GetNumLoopsRun))
  ;[1 1 1] Movement of trigger
  (Collision|ClearMoveIgnoreActors (Variables|Default|GetMovingTrigger))
  (Utilities|Set|Clear (Variables|Default|GetHitActors))
  (CallFunction|Calculate&SetWaveTime Start%20Delta)
  (CallFunction|SetTriggerVisualsVisibility true true)
  (if (Variables|Default|GetCurrentDirection)
    (bind (_direction _movement_alpha) (Timeline|TriggerMovement Play :NewTime (stale Custom|Set%20Wave%20Time NewTime))
      (:Update
        (Transformation|SetRelativeLocation (Variables|Default|GetMovingTrigger) (Math|Vector|Lerp(Vector) (CallFunction|GetStartDestinationforTrigger) (CallFunction|GetFinalDestinationforTrigger) _movement_alpha) (not (Variables|Default|GetbIsSettingTime)) false)
        (Variables|Default|SetCachedTriggerInstigator (Variables|Default|GetTriggerInstigator)))
      (:Finished
        ;;[1 1 1] Handles Direction Setting
        (if (and (or (Utilities|Enum|Equal(Enum) (Variables|UserOptions|Basic|GetPulse%20Direction) "NewEnumerator2") (Utilities|Enum|Equal(Enum) (Variables|UserOptions|Basic|GetPulse%20Direction) "NewEnumerator5")) (Utilities|Enum|Equal(Enum) _direction "Forward"))
          (CallFunction|ClearPauseLocation)
          (CallFunction|ReverseDirection)
          (CallFunction|ResumeWave)
          (elif (Default|GetCurrentLoopInfinitely)
            ;;[1 1 1] Handles Looping settings
            (Collision|SetCollisionEnabled (Variables|Default|GetMovingTrigger) "NoCollision")
            (else
              (Math|Integer|IncrementInt _numloopsrun
                (:%20%20
                  (if (>= _numloopsrun (Default|GetCurrentNumLoops))
                    (CallFunction|ResetDevice)
                    (else
                      _)))))))
        (CallFunction|ClearPauseLocation)
        (CallFunction|ReverseDirection)
        (Transformation|SetRelativeLocation (Variables|Default|GetMovingTrigger) (CallFunction|GetStartDestinationforTrigger) false false)
        (Collision|SetCollisionEnabled (Variables|Default|GetMovingTrigger) "QueryOnly")
        (CallFunction|ResumeWave)))
    (else
      (Timeline|TriggerMovement Reverse))))

(event Custom|StopWave
  (bind _numloopsrun (Variables|Default|GetNumLoopsRun))
  ;[1 1 1] Movement of trigger
  (Timeline|TriggerMovement Stop))

(event Custom|Set%20Wave%20Time
  (params (NewTime :type Float%20%28double-precision%29))
  (bind _numloopsrun (Variables|Default|GetNumLoopsRun))
  ;[1 1 1] Movement of trigger
  (Utilities|FlowControl|Sequence
    (:then_0
      ;;[1 1 1] Prevent sweeping
      (Variables|Default|SetbIsSettingTime true))
    (:then_1
      (Timeline|TriggerMovement SetNewTime))
    (:then_2
      (Variables|Default|SetbIsSettingTime false))))

(event Custom|StartSequenceFromEvent
  (params (Instigator :type Controller%20Object%20Reference))
  ;[0.9808 0 1] Gameplay Functions
  (if (Class|CreativeEnabledComponent|IsEnabled (Variables|Default|GetEnabledComponent))
    (if (Variables|UserOptions|GetSend%20Player%20Information)
      (Variables|Default|SetTriggerInstigator Instigator))
    (CallFunction|FireSequencerFromStart)))

(event OnEnabledStateChanged(EnabledComponent)
  (params (Enabled :type Boolean))
  ;[0 1 0.9331 0.25] Enabled State - When disabling, pause wave on server.  On server show or activate visibility.
  (CallFunction|UpdateTriggerCollision Enabled)
  (if (not Enabled)
    (CallFunction|ResetDevice)))

(event Custom|StopSequenceFromEvent
  (params (Instigator :type Controller%20Object%20Reference))
  ;[0.9808 0 1] Gameplay Functions
  (if (Class|CreativeEnabledComponent|IsEnabled (Variables|Default|GetEnabledComponent))
    (CallFunction|PauseWave)))

(event Custom|ResumeSequenceFromEvent
  (params (Instigator :type Controller%20Object%20Reference))
  ;[0.9808 0 1] Gameplay Functions
  (if (Class|CreativeEnabledComponent|IsEnabled (Variables|Default|GetEnabledComponent))
    (if (Variables|UserOptions|GetSend%20Player%20Information)
      (Variables|Default|SetTriggerInstigator Instigator))
    (CallFunction|ResumeWave)))

(event Custom|EnableFromEvent
  (params (Instigator :type Controller%20Object%20Reference))
  ;[0.9808 0 1] Gameplay Functions
  (Class|CreativeEnabledComponent|SetEnabled (Variables|Default|GetEnabledComponent) true))

(event Custom|DisableFromEvent
  (params (Instigator :type Controller%20Object%20Reference))
  ;[0.9808 0 1] Gameplay Functions
  (Class|CreativeEnabledComponent|SetEnabled (Variables|Default|GetEnabledComponent) false))

(event Custom|ActivateEventOnHitPawns
  (params (Instigator :type Controller%20Object%20Reference))
  ;[0.9808 0 1] Gameplay Functions
  (GameplayEvent|BroadcastEvent (Variables|UserOptions|Events|GetOnPlayerHit) 0 (Utilities|Struct|MakeGameplayEventonPlayerHit Instigator)))

(event PostMinigameStartedDelegate(FortMinigameProgress)
  (CreativeDeviceHelper|ActivateonPhase|BranchActivateonGameStart (Variables|UserOptions|GetActivateOnPhase)
    (:True
      (CallFunction|FireSequencerFromStart))))

(event EventOnMinigameEnded
  (Variables|verse|SetDamageOverride -1.0)
  (Variables|verse|SetLoopInfinitelyOverride "NewEnumerator0")
  (Variables|verse|SetNumLoopsOverride -1)
  (Variables|verse|SetTempoOverride -1.0))

(event MinigameStateChangedDelegate(FortMinigameProgress)
  (params (Minigame :type Fort%20Minigame%20Object%20Reference) (NewMinigameState :type EFortMinigameState%20Enum :default "PreGame"))
  (bind _triggerinstigator (Variables|Default|GetTriggerInstigator))
  (CallFunction|UpdateFrameVisibility)
  (CallFunction|ResetDevice)
  (Networking|SwitchHasAuthority
    (:Authority
      (Utilities|IsValid _triggerinstigator
        (:Is%20Valid
          (if (not (and (Collision|IsOverlappingActor (Variables|Default|GetPadTrigger) (Pawn|GetControlledPawn _triggerinstigator) :!class /Script/Engine.PrimitiveComponent) (Variables|UserOptions|GetSend%20Player%20Information)))
            (Variables|Default|SetTriggerInstigator 0 :!label _label))
          (if (Class|CreativeCommonDeviceFunctionLibrary2|ActivateonPhase Minigame (Variables|UserOptions|GetActivateOnPhase))
            (CallFunction|FireSequencerFromStart)))
        (:Is%20Not%20Valid
          _)) ; Check the instigator is valid so we don't spam the logs.
      (goto _label))))

(event EventInitializeDevice
  ;[0.0199 0.4896 0.1098] initialization
  (|CallParentFunction)
  (CallFunction|UpdateVolumeSize)
  (CallFunction|ResetDevice)
  (Networking|SwitchHasAuthority
    (:Authority
      (bind _triggerinstigator (Variables|Default|GetTriggerInstigator)
        (:then
          (if (not (and (Collision|IsOverlappingActor (Variables|Default|GetPadTrigger) (Pawn|GetControlledPawn _triggerinstigator) :!class /Script/Engine.PrimitiveComponent) (Variables|UserOptions|GetSend%20Player%20Information)))
            (Variables|Default|SetTriggerInstigator 0)))
        (:else
          _)) ; Validate, as on-load issues can cause this to return null.
      (if (Class|CreativeCommonDeviceFunctionLibrary|IsActivatedonPhase "PreGame" (Math|Conversions|ToInteger(Byte) (Variables|UserOptions|GetActivateOnPhase)) self)
        (CallFunction|FireSequencerFromStart)))))

(event EditorPreview|EventSetupForEditorPreview
  ;[0.0199 0.4896 0.1098] initialization
  (CallFunction|UpdateFrameVisibility)
  (CallFunction|UpdateVolumeSize)
  (CallFunction|ResetDevice)
  (Networking|SwitchHasAuthority
    (:Authority
      (bind _triggerinstigator (Variables|Default|GetTriggerInstigator)
        (:then
          (if (not (and (Collision|IsOverlappingActor (Variables|Default|GetPadTrigger) (Pawn|GetControlledPawn _triggerinstigator) :!class /Script/Engine.PrimitiveComponent) (Variables|UserOptions|GetSend%20Player%20Information)))
            (Variables|Default|SetTriggerInstigator 0)))
        (:else
          _)) ; Validate, as on-load issues can cause this to return null.
      (if (Class|CreativeCommonDeviceFunctionLibrary|IsActivatedonPhase "PreGame" (Math|Conversions|ToInteger(Byte) (Variables|UserOptions|GetActivateOnPhase)) self)
        (CallFunction|FireSequencerFromStart)))))

(event EventBeginPlay
  (CallFunction|ResetDevice)
  (Ability|EquipFortAbilitySet self "/Game/Creative/Devices/Common/GameplayAbilities/GAS_ApplyDamage.GAS_ApplyDamage"))

(event OnAddedtoMinigame(FortMinigameProgress)
  (params (Minigame :type Fort%20Minigame%20Object%20Reference))
  (CallFunction|UpdateFrameVisibility))

(event OnComponentBeginOverlap(MovingTrigger)
  (params (OverlappedComponent :type Primitive%20Component%20Object%20Reference) (OtherActor :type Actor%20Object%20Reference) (OtherComp :type Primitive%20Component%20Object%20Reference) (OtherBodyIndex :type Integer) (bFromSweep :type Boolean) (SweepResult :type Hit%20Result%20Structure%20%28by%20ref%29))
  ;[0.7195 1 0.6408] Moving Trigger's BeginOverlap
  (if (not (Utilities|Set|ContainsItem (Variables|Default|GetHitActors) OtherActor))
    (Utilities|Set|Add (Variables|Default|GetHitActors) OtherActor)
    (Collision|IgnoreActorwhenMoving (Variables|Default|GetMovingTrigger) OtherActor true)
    (Utilities|FlowControl|Sequence
      (:then_0
        (bind _asfort_music_sequencer_interface (Utilities|Casting|CastToFortMusicSequencerInterface OtherActor))
        (Class|FortMusicSequencerInterface|OnHitbyMusicSequencer _asfort_music_sequencer_interface (Variables|Default|GetCachedTriggerInstigator)))
      (:then_1
        (CallFunction|ApplyDamageToTarget OtherActor))
      (:then_2
        (CallFunction|DoonPlayerHitEvent OtherActor)))))

(fn ActivateFromOtherDevice
  (params (Instigator :type Controller%20Object%20Reference) (Requester :type ERequesterType%20Enum :default "NewEnumerator0"))
  (return false))

(fn ApplyDamagetoTarget
  (params (Actor :type Actor%20Object%20Reference))
  (Networking|SwitchHasAuthority
    (:Authority
      (Utilities|IsValid Actor
        (:Is%20Valid
          (Utilities|IsValid (Ability|GetAbilitySystemComponent self)
            (:Is%20Valid
              (if (and (CallFunction|HasMinigameStarted) (> (Default|GetCurrentDamage) 0))
                (Utilities|IsValid (Ability|GetAbilitySystemComponent Actor)
                  (:Is%20Valid
                    (Ability|SendGameplayEventtoActor self "(TagName=\"Abilities.Creative.ApplyDamage\")" (Utilities|Struct|MakeGameplayEventData "(TagName=\"Abilities.Creative.ApplyDamage\")" 0 (Variables|GetActor) 0 0 "()" "(GameplayTags=)" (Variables|Default|GetDamageTags) (Default|GetCurrentDamage)))))))))))))

(fn Calculate&SetWaveTime
  (params (Wave%20Start%20Time :type Float%20%28double-precision%29))
  (bind _tempo (Utilities|GetCurrentTempo))
  (bind _returnvalue (/ _tempo 6e+01))
  (bind _vertical (CallFunction|SweepDirection))
  (bind _height (Variables|UserOptions|Basic|GetHeight))
  (bind _length (Variables|UserOptions|Basic|GetLength))
  (bind _returnvalue_2 (select _vertical _height _length))
  (bind _returnvalue_3 (/ 1.0 _returnvalue_2))
  (bind _returnvalue_4 (* _returnvalue_3 _returnvalue))
  (Components|Timeline|SetPlayRate (Variables|Device_PulseTrigger|GetTriggerMovement) _returnvalue_4) ; Keeps playback rate consistent regardless of length of trigger
  (CallFunction|SetWaveTime (Math|Float|SelectFloat (+ (CallFunction|GetBaseLocationAlpha) (* _returnvalue_4 Wave%20Start%20Time)) (- (CallFunction|GetBaseLocationAlpha) (* _returnvalue_4 Wave%20Start%20Time)) (Variables|Default|GetCurrentDirection))))

(fn CanInteract
  (params (InteractingPawn :type Fort%20Pawn%20Object%20Reference) (InteractionBeingAttempted :type EInteractionBeingAttempted%20Enum :default "FirstInteraction") (InteractionType :type TInteractionType%20Enum :default "IT_NoInteraction"))
  (return false))

(fn ClearPauseLocation
  (Networking|SwitchHasAuthority
    (:Authority
      (if (!= (Variables|Default|GetServerWavePauseLocation) -1.0)
        (Variables|Default|SetServerWavePauseLocation -1.0)))))

(fn DoOnPlayerHitEvent
  (params (Actor :type Actor%20Object%20Reference))
  (bind (_asfort_vehicle_interface _bsuccess) (Utilities|Casting|CastToFortVehicleInterface Actor))
  (bind _returnvalue (Vehicle|GetAllPassengers _asfort_vehicle_interface))
  (if _bsuccess
    ;[1 1 1] If in a vehicle let's tag all the passengers.
    (if (Utilities|Array|IsNotEmpty _returnvalue)
      (for _array_element _returnvalue
        (Utilities|Set|Add (Variables|Default|GetHitActors) _array_element)
        (CallFunction|ActivateEventonHitPawns (Pawn|GetController _array_element))))
    (else
      (bind _asfort_pawn (Utilities|Casting|CastToFortPawn Actor))
      (CallFunction|ActivateEventonHitPawns (Pawn|GetController _asfort_pawn)))))

(fn GetBaseLocationAlpha
  (return (select (!= (Variables|Default|GetServerWavePauseLocation) -1.0) (Variables|Default|GetServerWavePauseLocation) (select (Variables|Default|GetCurrentDirection) 0.0 1.0))))

(fn GetDeviceColor
  (return 0))

(fn GetEnabledStateasIndex
  (return (DeviceConversion|EnabledPhaseEnumtoLegacyInt (Variables|UserOptions|GetEnabled%20on%20Phase))))

(fn GetInstigatingController
  (return (Variables|Default|GetTriggerInstigator)))

(fn GetCollisionStaticMesh
  (return "0, 0, 0" 0))

(fn GetDamage
  (return (Default|GetCurrentDamage)))

(fn GetFinalDestinationForTrigger
  ;[1 1 1] Get the distance trigger needs to travel based on scale of volume
  (return (Math|Vector|MakeVector (* (Variables|UserOptions|Basic|GetLength) (select (CallFunction|SweepDirection) 256.0 512.0)) 0.0 (* (Math|Conversions|ToFloat(Boolean) (CallFunction|SweepDirection)) (* (Variables|UserOptions|Basic|GetHeight) 384.0)))))

(fn GetMeshComponents
  (return 0))

(fn GetNewLocation
  (return (Math|Vector|MakeVector (+ (* (- (Variables|UserOptions|Basic|GetLength) 1.0) 256.0) (.x (Variables|Default|GetStartingLocationOfVolume))) (.y (Variables|Default|GetStartingLocationOfVolume)) (.z (Variables|Default|GetStartingLocationOfVolume)))))

(fn GetNumLoops
  (return (Default|GetCurrentNumLoops)))

(fn GetScaleMax
  (params (DefaultScaleMax :type Float%20%28single-precision%29))
  (return 1.0))

(fn GetScaleMin
  (params (DefaultScaleMin :type Float%20%28single-precision%29))
  (return 1.0))

(fn GetSpawnedPickups
  (return 0))

(fn GetStartDestinationForTrigger
  ;[1 1 1]
  (return (Utilities|Select "0, 0, 0" (Math|Vector|MakeVector (* (Variables|UserOptions|Basic|GetLength) 256.0)) (CallFunction|SweepDirection))))

(fn GetTempo
  (return (Utilities|GetCurrentTempo)))

(fn GetWaveSpeed
  (return (* (/ (Utilities|GetCurrentTempo) 6e+01) 5.12)))

(fn HasMinigameStarted
  (bind _currentminigame (Variables|/Script/FortniteGame.FortMinigameLogicComponent|GetCurrentMinigame (Variables|Default|GetFortMinigameProgress))
    (:then
      (return (Minigame|HasStarted _currentminigame :!class /Script/FortniteGame.FortMinigame)))
    (:else
      (return false))))

(fn IgnoreWorldGridScaleOverride
  (return false))

(fn IsEnabled
  (return false))

(fn IsHoldingItemForPawn
  (params (Item :type Item%20Variant%20Handle%20Structure) (Pawn :type Pawn%20Object%20Reference))
  (return false))

(fn IsMinigameOnPregameState
  (bind _currentminigame (Variables|/Script/FortniteGame.FortMinigameLogicComponent|GetCurrentMinigame (Variables|Default|GetFortMinigameProgress))
    (:then
      (return (Utilities|Enum|Equal(Enum) (Variables|/Script/FortniteGame.FortMinigame|GetCurrentState _currentminigame) "PreGame")))
    (:else
      (return (Editor|InEditor)))))

(fn IsPulseActive
  (return (and (or (and (>= (Variables|Default|GetServerWaveStartTime) 0) (< (Variables|Default|GetServerWavePauseLocation) 0)) (and (>= (Variables|Default|GetServerWavePauseLocation) 0) (Variables|UserOptions|GetActive%20when%20Paused))) (Class|CreativeEnabledComponent|IsEnabled (Variables|Default|GetEnabledComponent)))))

(fn IsTargetableByLinearMover
  (return true))

(fn MigrateLegacyUserOptions
  (bind (_number_of_loops _startsequence _numloopssetting) (Class|GetClassDefaults "/CreativeCoreDevices/Device_PulseTrigger.Device_PulseTrigger_C"))
  (bind _numloopssetting_2 (Variables|LegacyMigration|GetNumLoopsSetting))
  (|CallParentFunction)
  (Utilities|FlowControl|Sequence
    (:then_0
      ;[1 1 1] Num Loop Settings
      (if (!= _numloopssetting _numloopssetting_2)
        (bind _output_get (Variables|UserOptions|Basic|SetLoop%20Infinitely (== _numloopssetting_2 0)))
        (if _output_get
          (Variables|UserOptions|Basic|SetNumber%20of%20Loops _number_of_loops)
          (else
            (Variables|UserOptions|Basic|SetNumber%20of%20Loops _numloopssetting_2)))
        (Variables|LegacyMigration|SetNumLoopsSetting _numloopssetting)
        (LocalVariables|SetOption%20Migrated true)))
    (:then_1
      ;[1 0.1681 0.9838 0.25] If any options have migrated, refresh save data from this instance's uproperties.
      (if (LocalVariables|GetOption%20Migrated)
        (OptionData|RefreshPropertyOverrides (Variables|Default|GetToyOptionsComponent))))))

(fn OnRep_DamageOverride
  (Networking|CanExecuteCosmeticEvents
    (:True
      (CallFunction|UpdateDamageMaterialParameters))))

(fn OnRep_ServerPauseWaveTime
  (if (!= (Variables|Default|GetServerWavePauseLocation) -1.0)
    (CallFunction|SetWaveTime (CallFunction|GetBaseLocationAlpha))
    (CallFunction|StopWave)
    (CallFunction|SetTriggerVisualsVisibility (and (Variables|UserOptions|GetActive%20when%20Paused) (Class|CreativeEnabledComponent|IsEnabled (Variables|Default|GetEnabledComponent))))))

(fn OnRep_ServerWaveStartTime
  (if (== (Variables|Default|GetServerWaveStartTime) 0.0)
    ;[1 0.03568 0] Reset Logic
    (CallFunction|StopWave)
    (CallFunction|SetTriggerVisualsVisibility false)
    (CallFunction|ReturnWavetoStartLocation)
    (Variables|Default|SetInitialTrigger true) ; Reset so player activation will have sound again
    (else
      ;[0.0199 0.4896 0.1098] Start Logic
      (Variables|Default|SetCurrentDirection (>= (Variables|Default|GetServerWaveStartTime) 0.0))
      (CallFunction|OnWaveFired (- (GameState|GetServerWorldTimeSeconds (Game|GetGameState)) (Math|Float|Absolute(Float) (Variables|Default|GetServerWaveStartTime)))))))

(fn PauseWave
  ;[1 1 1] Stuff to help ensure clients are sync'd up as wave travels
  (Networking|SwitchHasAuthority
    (:Authority
      (Networking|FlushNetDormancy)
      (Variables|Default|SetServerWavePauseLocation (Components|Timeline|GetPlaybackPosition (Variables|Device_PulseTrigger|GetTriggerMovement)))
      (Networking|ForceNetUpdate) ; Replicate ASAP
    )))

(fn PlayAudio
  (Networking|CanExecuteCosmeticEvents
    (:True
      (if (Variables|Default|GetInitialTrigger)
        (Audio|PlaySoundatLocation (Variables|Default|GetStart%20Sequence%20Sound) (Transformation|GetActorLocation))
        (Variables|Default|SetInitialTrigger false)))))

(fn ResetDevice
  (Networking|SwitchHasAuthority
    (:Authority
      (if (!= (Variables|Default|GetServerWaveStartTime) 0.0)
        (Variables|Default|SetServerWaveStartTime 0.0) ; Setting a value of 0 will cause the device to reset.
      ))
    (:Remote
      (CallFunction|SetTriggerVisualsVisibility false))))

(fn ResumeWave
  ;[1 1 1] Stuff to help ensure clients are sync'd up as wave travels
  (Networking|SwitchHasAuthority
    (:Authority
      (Networking|FlushNetDormancy)
      (Variables|Default|SetServerWaveStartTime (* (GameState|GetServerWorldTimeSeconds (Game|GetGameState)) (Math|Float|SelectFloat 1.0 -1.0 (Variables|Default|GetCurrentDirection)))) ; Notify that a wave has been fired
      (Networking|ForceNetUpdate) ; Replicate ASAP
    )))

(fn ReturnWaveToStartLocation
  (Variables|Default|SetNumLoopsRun 0) ; reset loop count
  (CallFunction|ClearPauseLocation)
  (if (Variables|Default|GetCurrentDirection)
    (Transformation|SetRelativeLocation (Variables|Default|GetMovingTrigger) (CallFunction|GetFinalDestinationforTrigger) false false) ; Get trigger back to start
    (else
      (Transformation|SetRelativeLocation (Variables|Default|GetMovingTrigger) (CallFunction|GetStartDestinationforTrigger) false false) ; Get trigger back to start
    )))

(fn ReverseDirection
  (if (or (Utilities|Enum|Equal(Enum) (Variables|UserOptions|Basic|GetPulse%20Direction) "NewEnumerator2") (Utilities|Enum|Equal(Enum) (Variables|UserOptions|Basic|GetPulse%20Direction) "NewEnumerator5"))
    (Variables|Default|SetCurrentDirection (not (Variables|Default|GetCurrentDirection))) ; Bounce in the other direction. This isn't replicated directly so it can be client inferred to avoid latency when using the bounce feature.
  ))

(fn SetInitialDirection
  (Variables|Default|SetCurrentDirection (not (or (Utilities|Enum|Equal(Enum) (Variables|UserOptions|Basic|GetPulse%20Direction) "NewEnumerator1") (Utilities|Enum|Equal(Enum) (Variables|UserOptions|Basic|GetPulse%20Direction) "NewEnumerator4")))))

(fn SetTriggerCollision
  (params (On :type Boolean))
  (return false))

(fn SetDamage
  (params (Damage :type Float%20%28double-precision%29))
  (bind _returnvalue (Math|Float|Clamp(Float) Damage 0.0 1e+05))
  (Networking|SwitchHasAuthority
    (:Authority
      (if (!= _returnvalue (Variables|verse|GetDamageOverride))
        (Variables|verse|SetDamageOverride _returnvalue)
        (Networking|FlushNetDormancy)))))

(fn SetNumLoops
  (params (NumLoops :type Integer))
  (bind _returnvalue (Math|Integer|Clamp(Integer) NumLoops 0 1000))
  (Networking|SwitchHasAuthority
    (:Authority
      (if (!= (Variables|verse|GetNumLoopsOverride) _returnvalue)
        (Networking|SwitchHasAuthority
          (:Authority
            (bind _output_get (Variables|verse|SetNumLoopsOverride _returnvalue))
            ;[1 1 1] Treat 0 loop count as infinite
            (Variables|verse|SetLoopInfinitelyOverride (select (== _output_get 0) "NewEnumerator2" "NewEnumerator1"))
            (Networking|FlushNetDormancy)))))))

(fn SetTempo
  (params (Tempo :type Float%20%28double-precision%29))
  (bind _returnvalue (Math|Float|Clamp(Float) Tempo 1.0 1.8e+02))
  (Networking|SwitchHasAuthority
    (:Authority
      (if (!= _returnvalue (Variables|verse|GetTempoOverride))
        (Networking|SwitchHasAuthority
          (:Authority
            (Variables|verse|SetTempoOverride _returnvalue)
            (Networking|FlushNetDormancy)))))))

(fn SetTriggerVisualsVisibility
  (params (bNewVisibility :type Boolean) (IlluminateFloorPad :type Boolean))
  (bind _pressureplate (Variables|Default|GetPressurePlate))
  (bind _mid (Material|GetMID _pressureplate))
  (bind _movingtrigger (Variables|Default|GetMovingTrigger))
  (Rendering|SetVisibility _movingtrigger bNewVisibility true)
  (Collision|SetCollisionEnabled _movingtrigger (select bNewVisibility "QueryOnly" "NoCollision"))
  (Components|Activation|SetActive (Variables|Default|GetP_Sequencer_Blocks) bNewVisibility false)
  (Rendering|Material|SetVectorParameterValue _mid "GlowColor" (select IlluminateFloorPad (Variables|Default|GetFloorPadGlowColor) "(R=0.000000,G=0.000000,B=0.000000,A=1.000000)") :!class /Script/Engine.MaterialInstanceDynamic)
  (Rendering|Material|SetTextureParameterValue _mid "Texture" (select IlluminateFloorPad (Variables|Default|GetTextureOn) (Variables|Default|GetTextureOff))))

(fn SetWaveSpeed
  (params (SpeedMetersPerSecond :type Float%20%28double-precision%29))
  (bind _returnvalue (/ SpeedMetersPerSecond 5.12))
  (bind _returnvalue_2 (* _returnvalue 6e+01))
  (bind _returnvalue_3 (Math|Float|Clamp(Float) _returnvalue_2 1.0 1.8e+02))
  (Networking|SwitchHasAuthority
    (:Authority
      (if (!= _returnvalue_3 (Variables|verse|GetTempoOverride))
        (Networking|SwitchHasAuthority
          (:Authority
            (Variables|verse|SetTempoOverride _returnvalue_3)
            (Networking|FlushNetDormancy)))))))

(fn ShouldInvisiblePropsBeAllowed
  (return false))

(fn SweepDirection
  (return (Utilities|Select :Index (Variables|UserOptions|Basic|GetPulse%20Direction) :NewEnumerator0 false :NewEnumerator1 false :NewEnumerator2 false :NewEnumerator3 true :NewEnumerator4 true :NewEnumerator5 true)))

(fn ToggleLogic
  (if (Components|Timeline|IsPlaying (Variables|Device_PulseTrigger|GetTriggerMovement))
    (CallFunction|ResetDevice)
    (else
      (CallFunction|FireSequencerFromStart))))

(fn TogglePauseLogic
  ;[1 1 1] Stuff to help ensure clients are sync'd up as wave travels
  (if (Components|Timeline|IsPlaying (Variables|Device_PulseTrigger|GetTriggerMovement))
    (CallFunction|PauseWave)
    (elif (== (Variables|Default|GetServerWavePauseLocation) -1.0)
      (CallFunction|FireSequencerFromStart)
      (else
        (CallFunction|ResumeWave)))))

(fn UpdateFrameVisibility
  (bind _frame (Variables|Default|GetFrame))
  (Rendering|SetVisibility _frame (or (Math|Boolean|ANDBoolean (not (Replay|GetIsPlayingReplay)) (not (CallFunction|IsinPlayMode self)) (CallFunction|IsMinigameonPregameState)) (Variables|UserOptions|Basic|GetZone%20Visible%20During%20Game)) false)
  (Development|SetHiddeninGame _frame (and (not (Variables|UserOptions|Basic|GetZone%20Visible%20During%20Game)) (Editor|InEditor)) false) ; Set Hidden in Game for Editor only if we're not visible in game\n(so Game Mode will hide frame)
  (return))

(fn UpdateTriggerCollision
  (params (Enable :type Boolean))
  ;[0 1 0.9331 0.25] Enable or disable the collision on the triggers, allowing or disallowing activation by collision.
  (Collision|SetCollisionEnabled (Variables|Default|GetChainTrigger) (select Enable "QueryOnly" "NoCollision")))

(fn UpdateDamageMaterialParameters
  (bind _damage (Default|GetCurrentDamage))
  (bind _damage_2 (Default|GetCurrentDamage))
  (bind _returnvalue (> _damage_2 0))
  (bind _returnvalue_2 (> _damage 0))
  (bind _movingtrigger (Variables|Default|GetMovingTrigger))
  (bind _movingtriggermid (Variables|Default|GetMovingTriggerMID))
  (if (not (Variables|UserOptions|Advanced|GetCustomPulseStyle))
    (Rendering|Material|SetMaterial (Variables|Default|GetMovingTrigger) 0 (select _returnvalue "/Game/Items/Traps/Materials/M_Sequencer_Trigger_02_Inst.M_Sequencer_Trigger_02_Inst" "/Game/Items/Traps/Materials/M_Sequencer_Trigger_01_Inst.M_Sequencer_Trigger_01_Inst"))
    (Rendering|Material|SetMaterial (Variables|Default|GetP_Sequencer_Blocks) 0 (select _returnvalue "/Game/Items/Traps/Materials/M_EliminationZone_Blocks.M_EliminationZone_Blocks" "/Game/Items/Traps/Materials/M_Sequencer_Blocks_Inst.M_Sequencer_Blocks_Inst"))
    (Rendering|Material|SetScalarParameterValueonMaterials (Variables|Default|GetMovingTrigger) "UseFlashingAlert" (select (and (Variables|UserOptions|Advanced|GetUseFlashingWarningSigns) _returnvalue) 1.0 0.0))
    (return)
    (else
      (Rendering|Material|SetMaterial _movingtrigger 0 "/Game/Creative/Devices/Sequencer/Materials/MI_Device_Sequencer_Volume_01.MI_Device_Sequencer_Volume_01")
      (Material|ReattachOrCreateMID _movingtrigger _movingtriggermid)
      (Rendering|Material|SetVectorParameterValue _movingtriggermid "Base Color" (select _returnvalue_2 "(R=0.901042,G=0.015019,B=0.000000,A=0.000000)" "(R=0.000000,G=0.040255,B=0.901042,A=0.000000)") :!class /Script/Engine.MaterialInstanceDynamic)
      (Rendering|Material|SetVectorParameterValue _movingtriggermid "Swirls Color" (select _returnvalue_2 "(R=1.000000,G=0.108540,B=0.000000,A=0.000000)" "(R=0.027170,G=0.789307,B=1.000000,A=0.000000)") :!class /Script/Engine.MaterialInstanceDynamic)
      (Rendering|Material|SetVectorParameterValue _movingtriggermid "Frame Glow Color" (select _returnvalue_2 "(R=1.000000,G=0.102506,B=0.000000,A=1.000000)" "(R=0.000000,G=0.818932,B=1.000000,A=1.000000)") :!class /Script/Engine.MaterialInstanceDynamic)
      (Rendering|Material|SetScalarParameterValue _movingtriggermid "Width" (* (Variables|UserOptions|Basic|GetWidth) 512.0) :!class /Script/Engine.MaterialInstanceDynamic)
      (Rendering|Material|SetScalarParameterValue _movingtriggermid "Height" (* (Variables|UserOptions|Basic|GetHeight) 384.0) :!class /Script/Engine.MaterialInstanceDynamic)
      (Rendering|Material|SetScalarParameterValue _movingtriggermid "Depth" (* (Variables|UserOptions|Basic|GetLength) 512.0) :!class /Script/Engine.MaterialInstanceDynamic))))

(fn UpdateVolumeSize
  (bind _p_sequencer_blocks (Variables|Default|GetP_Sequencer_Blocks))
  (Components|Activation|Deactivate (Variables|Default|GetP_Sequencer_Blocks))
  (Rendering|Material|SetVectorParameterValue (Material|GetMID (Variables|Default|GetFrame)) "SizeBias" (Math|Conversions|ToLinearColor(Vector) (- (* (Math|Vector|MakeVector 256.0 256.0 384.0) (Math|Vector|MakeVector (Variables|UserOptions|Basic|GetLength) (Variables|UserOptions|Basic|GetWidth) (Variables|UserOptions|Basic|GetHeight))) (Math|Vector|MakeVector 256.0 256.0 384.0))) :!class /Script/Engine.MaterialInstanceDynamic)
  (Rendering|SetBoundsScale (Variables|Default|GetFrame) (* (Math|Float|Max(Float) (Variables|UserOptions|Basic|GetLength) (Variables|UserOptions|Basic|GetWidth) (Variables|UserOptions|Basic|GetHeight)) 1.5))
  (Transformation|SetWorldScale3D (Variables|Default|GetMovingTrigger) (select (CallFunction|SweepDirection) (Math|Vector|MakeVector (select (Variables|UserOptions|Advanced|GetCustomPulseStyle) (* (Variables|UserOptions|Basic|GetLength) 512.0) (Variables|UserOptions|Basic|GetLength)) (select (Variables|UserOptions|Advanced|GetCustomPulseStyle) (* (Variables|UserOptions|Basic|GetWidth) 512.0) (Variables|UserOptions|Basic|GetWidth)) (select (Variables|UserOptions|Advanced|GetCustomPulseStyle) 25.0 0.05)) (Math|Vector|MakeVector (select (Variables|UserOptions|Advanced|GetCustomPulseStyle) 25.0 0.05) (select (Variables|UserOptions|Advanced|GetCustomPulseStyle) (* (Variables|UserOptions|Basic|GetWidth) 512.0) (Variables|UserOptions|Basic|GetWidth)) (select (Variables|UserOptions|Advanced|GetCustomPulseStyle) (* (Variables|UserOptions|Basic|GetHeight) 384.0) (Variables|UserOptions|Basic|GetHeight)))))
  (Transformation|SetRelativeLocation (Variables|Default|GetMovingTrigger) (CallFunction|GetStartDestinationforTrigger) false false)
  (Transformation|SetRelativeLocation (Variables|Default|GetFrame) (CallFunction|GetNewLocation) false false)
  (Variables|/Script/CascadeRuntime.ParticleSystemComponent|SetInstanceParameters (CallFunction|VFXParameters) (Variables|Default|GetP_Sequencer_Blocks))
  (Transformation|SetWorldScale3D _p_sequencer_blocks "1.000000,1.000000,1.000000")
  (if (CallFunction|SweepDirection)
    (Transformation|SetRelativeLocation _p_sequencer_blocks "256.000000,256.000000,0.000000" false false)
    (else
      (Transformation|SetRelativeLocation _p_sequencer_blocks "0,256.000000,384.000000" false false)))
  (Components|Activation|Activate (Variables|Default|GetP_Sequencer_Blocks) true)
  (Components|StaticMesh|SetStaticMesh (Variables|Default|GetMovingTrigger) (select (Variables|UserOptions|Advanced|GetCustomPulseStyle) "/Game/Creative/Devices/Sequencer/Meshes/CP_Device_Sequencer_Volume_01.CP_Device_Sequencer_Volume_01" "/Game/Creative/Devices/Music_Sequencer/S_MusicSequencer_Wave.S_MusicSequencer_Wave"))
  (CallFunction|UpdateDamageMaterialParameters)
  (Transformation|SetRelativeTransform (Variables|Default|GetTriggerBase) (Utilities|Map|Find (Variables|Default|GetZoneDirectionsMap) (Variables|UserOptions|GetZoneDirection)) false true))

(fn ConstructionScript
  (Networking|CanExecuteCosmeticEvents
    (:True
      (Variables|/Script/Engine.SceneComponent|SetbAbsoluteScale true (Variables|Default|GetP_Sequencer_Blocks)))))

(fn ValidateandSetInstigator
  (params (OtherActor :type Actor%20Object%20Reference))
  (Utilities|FlowControl|Sequence
    (:then_0
      ;[1 1 1] Check for valid actor overlap
      (bind _asplayer_pawn_athena (Utilities|Casting|CastToPlayerPawn_Athena (Variables|GetOtherActor))
        (:then
          (if (not (Vehicle|IsinVehicle _asplayer_pawn_athena))
            (LocalVariables|SetLocalIsActorValid true)
            (LocalVariables|SetLocalInstigator (Pawn|GetController _asplayer_pawn_athena))))
        (:CastFailed
          (bind _asfort_athena_vehicle (Utilities|Casting|CastToFortAthenaVehicle (Variables|GetOtherActor))
            (:then
              (LocalVariables|SetLocalIsActorValid true)
              (bind _returnvalue (Vehicle|GetDriver(Message) _asfort_athena_vehicle))
              (Utilities|IsValid _returnvalue
                (:Is%20Valid
                  (LocalVariables|SetLocalInstigator (Pawn|GetController _returnvalue)))
                (:Is%20Not%20Valid
                  (bind _returnvalue_2 (Vehicle|GetAllPassengers(Message) _asfort_athena_vehicle))
                  (bind _array_element (Utilities|Array|ForEachLoopwithBreak _returnvalue_2 :!label _label)
                    (:LoopBody
                      (Utilities|IsValid _array_element
                        (:Is%20Valid
                          (Utilities|IsValid (Pawn|GetController _array_element)
                            (:Is%20Valid
                              (LocalVariables|SetLocalInstigator (Pawn|GetController _array_element))
                              (goto _label Break))))))))))
            (:CastFailed
              (bind _asphysics_ball_master (Utilities|Casting|CastToPhysicsBall_Master (Variables|GetOtherActor))
                (:then
                  (LocalVariables|SetLocalIsActorValid true)
                  (LocalVariables|SetLocalInstigator (Pawn|GetController (Variables|Actor|GetInstigator _asphysics_ball_master))))
                (:CastFailed
                  (bind _asfort_physics_pawn_object (Utilities|Casting|CastToFortPhysicsPawnObject (Variables|GetOtherActor)))
                  (LocalVariables|SetLocalIsActorValid true)
                  (LocalVariables|SetLocalInstigator (Variables|/Script/Engine.Pawn|GetLastHitBy _asfort_physics_pawn_object)))))))))
    (:then_1
      ;[1 1 1] Assign instigator and bind delegate
      (if (and (LocalVariables|GetLocalIsActorValid) (Variables|UserOptions|GetSend%20Player%20Information))
        (Variables|Default|SetTriggerInstigator (LocalVariables|GetLocalInstigator)))
      (return (LocalVariables|GetLocalIsActorValid)))))

(fn VFXParameters
  (return (Utilities|Array|MakeArray :%5B0%5D (Utilities|Struct|MakeParticleSysParam "BlocksVolume" "PSPT_VectorRand" 0.0 0.0 "0.000000,0.000000,0.000000" (* (* (select (CallFunction|SweepDirection) (Variables|Default|GetBlocksVolumeVertical) (Variables|Default|GetBlocksVolumeHorizontal)) (Math|Vector|MakeVector (Variables|UserOptions|Basic|GetLength) (Variables|UserOptions|Basic|GetWidth) (Variables|UserOptions|Basic|GetHeight))) -1.0)) :%5B1%5D (Utilities|Struct|MakeParticleSysParam "BlocksSpawnRateScale" "PSPT_Scalar" (Utilities|ClampMaxFloat (* (Math|Float|SelectFloat (* (Variables|UserOptions|Basic|GetLength) (Variables|UserOptions|Basic|GetWidth)) (* (Variables|UserOptions|Basic|GetWidth) (Variables|UserOptions|Basic|GetHeight)) (CallFunction|SweepDirection)) (Variables|Default|GetFX%20Spawn%20Rate%20per%20Unit%20Squared)) (Variables|Default|GetFX%20Max%20Spawn%20Rate))))))
