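; Fortnite Creative radio device that plays a chosen track at the device or on registered players, driven by minigame phases and Play/Stop receivers.
; The server sets TimeSongStarted and Instigators, whose OnReps update client playback, optionally synced by server-time offset.
; Clients async-load the track, reuse or spawn pawn-attached audio components, fade them, and feed them to AudioToMPC visualization.
; It also drives mesh material pulses and migrates legacy options.
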(event EditorPreview|EventSetupForEditorPreview
  (|CallParentFunction)
  (CallFunction|UpdateMeshAndMaterial))

(event Custom|Update%20Mesh%20and%20Material
  (bind _staticmeshcomponent (Variables|Debug|GetStaticMeshComponent))
  ;[1 0.8394 0 0.1] Object Initialization Optimization: If we're in play mode and don't show in-game, don't load the mesh.
  (DeviceResetting|ResetDevice)
  (if (or (not (CallFunction|IsinPlayMode)) (ChildFunctions|IsVisibleInGame))
    (bind (_static_mesh _material) (Utilities|GetMeshandMaterial)
      (:%20%20
        (Components|StaticMesh|SetStaticMesh _staticmeshcomponent _static_mesh)
        (Rendering|Material|SetMaterial _staticmeshcomponent 0 _material) ; Reset Material
        (CallFunction|UpdateVisuals)
        (Development|PrintString (Utilities|String|Append "UPDATED TO: " (Utilities|GetDisplayName _static_mesh)))))
    (else
      (Components|StaticMesh|SetStaticMesh (Variables|Debug|GetStaticMeshComponent) 0))))

(event EventInitializeDevice
  ;[0 1 0.06377 0.208] Device Initialization
  (|CallParentFunction)
  (CallFunction|SetupAudioComponent)
  (CallFunction|UpdateMeshAndMaterial)
  (Networking|CanExecuteCosmeticEvents
    (:True
      (DeviceInitialization|SetAttenuationDistance)
      (DeviceInitialization|SetVisualizationState))))

(event ChildFunctions|EventPlayModeChanged
  (params (Minigame :type Fort%20Minigame%20Object%20Reference) (bIsInPlayMode :type Boolean))
  ;[0 1 0.06377 0.208] Device Initialization
  (|CallParentFunction Minigame bIsInPlayMode)
  (CallFunction|UpdateMeshAndMaterial))

(event Custom|Setup%20Audio%20Component
  (bind _audio_component (Variables|Default|GetAudio%20Component))
  ;[0 1 0.1819 0.25] Load Audio and Play it.
  (Creative|Networking|Authority/Cosmetic
    (:Authority
      (if (!= (Variables|Default|GetServer%20Cached%20Audio) (Variables|UserOptions|Basic|GetAudio))
        (DeviceResetting|ResetDevice) ; If we change tracks, we want to reset our audio to play from the start
        (Variables|Default|SetServer%20Cached%20Audio (Variables|UserOptions|Basic|GetAudio))))
    (:Cosmetic
      (if (Utilities|NotEqual(SoftObjectReference) (Variables|UserOptions|Basic|GetAudio) (Utilities|ResolveSoftReference (Variables|/Script/Engine.AudioComponent|GetSound (Variables|Default|GetAudio%20Component))))
        (bind _loaded_object (Creative|Loading|AsyncLoadEditorSafe (Variables|UserOptions|Basic|GetAudio))
          (:Loaded
            (bind _assound_base (Utilities|Casting|CastToSoundBase _loaded_object))
            (Audio|Components|Audio|SetSound _audio_component _assound_base))))
      (Audio|Components|Audio|SetPitchMultiplier _audio_component (Variables|UserOptions|GetPlaybackSpeed))
      (Audio|Components|Audio|SetVolumeMultiplier _audio_component (Variables|UserOptions|Basic|GetVolume) :!class /Script/Engine.AudioComponent)
      (Variables|Default|SetClient%20Cached%20Audio (Variables|UserOptions|Basic|GetAudio))
      (Utilities|IsAudioPlaying
        (:Yes
          (Client|AudioPlaying|ClientUpdateSongState false))) ; If our audio should be playing,\nplay it now (as it would have failed\ndue to the audio track not being loaded.
    )))

(event MinigameStateChangedDelegate(FortMinigameProgress)
  (params (Minigame :type Fort%20Minigame%20Object%20Reference) (NewMinigameState :type EFortMinigameState%20Enum :default "PreGame"))
  ;[0 1 0.9331 0.25] Minigame State Change.
  (bind _wants_to_play (Utilities|IsAudioPhaseEnabled NewMinigameState)
    (:Update%20State
      (Server|SetAudioState _wants_to_play))))

(event Custom|Event%20On%20Receive%20Unregister%20All%20Players
  (params (Instigator :type Controller%20Object%20Reference))
  ;[0.4774 0 3 0.2] Gameplay Functions
  (Variables|TransientData|PlayerTracking|SetInstigators (Utilities|Array|MakeArray)))

(event Custom|Event%20On%20Receive%20Play
  (params (Instigator :type Controller%20Object%20Reference))
  ;[0.4774 0 3 0.2] Gameplay Functions
  (Server|SetAudioState true (Variables|UserOptions|GetRestart%20Audio%20On%20Activation)))

(event Custom|Event%20On%20Receive%20Register%20Player
  (params (Instigator :type Controller%20Object%20Reference))
  ;[0.4774 0 3 0.2] Gameplay Functions
  (Server|UpdateRegisteredPlayers true Instigator))

(event Custom|Event%20On%20Receive%20Stop
  (params (Instigator :type Controller%20Object%20Reference))
  ;[0.4774 0 3 0.2] Gameplay Functions
  (Server|SetAudioState false))

(event Custom|Event%20On%20Receive%20Unregister%20Player
  (params (Instigator :type Controller%20Object%20Reference))
  ;[0.4774 0 3 0.2] Gameplay Functions
  (Server|UpdateRegisteredPlayers false Instigator))

(fn ClientUpdateSongState
  (params (bShouldRestartSounds :type Boolean))
  (if (not (Editor|InEditor))
    (Networking|CanExecuteCosmeticEvents
      (:True
        (Utilities|FlowControl|Sequence
          (:then_0
            (if bShouldRestartSounds
              (Client|AudioPlaying|StopSounds))
            (Utilities|ShouldPlaySong
              (:Yes
                (Variables|/Script/Engine.AudioComponent|GetSound (Variables|Default|GetAudio%20Component)
                  (:then
                    (switch Utilities|FlowControl|Switch|SwitchonE_PlayAtLocationCreativeRadio (Variables|UserOptions|GetPlayAtLocation)
                      (:NewEnumerator0
                        (Client|AudioPlaying|PlaySoundatDevice))
                      (:NewEnumerator1
                        (if (ScalableFloat|GetValueasBool (Variables|Config|GetbAllowInstigatorLocationBasedAudio))
                          (for _array_element (Variables|TransientData|PlayerTracking|GetInstigators)
                            (Client|AudioPlaying|PlaySoundOnPlayer _array_element))))))
                  (:else)) ; Only continue if we have a loaded song
              )
              (:No
                (Client|AudioPlaying|StopSounds))))
          (:then_1
            (CallFunction|UpdateVisuals)))))))

(fn DestroyFadingPlayerSoundComponents
  (bind _fading_out_sounds (Variables|TransientData|PlayerTracking|GetFading%20Out%20Sounds))
  ;[1 0.3072 0 0.2] Destroy all sounds in the fading out sounds list, as they will now have faded out (or the device wants to reset).
  (bind _values (Utilities|Map|Values _fading_out_sounds))
  (for _array_element _values
    (Utilities|IsValid _array_element
      (:Is%20Valid
        (AudioToMPC|RemoveTrackedAudioComponent (Variables|Default|GetCreative_Radio_AudioToMPC) _array_element)
        (Components|DestroyComponent _array_element))))
  (Utilities|Map|Clear _fading_out_sounds)
  (CallFunction|UpdateVisuals))

(fn GetStartTime
  (bind _sound (Variables|/Script/Engine.AudioComponent|GetSound (Variables|Default|GetAudio%20Component))
    (:then
      (return (Math|Float|SelectFloat (Math|Float|%25%28Float%29 (- (Fortnite|Sync|Time|GetServerWorldTimeSeconds) (Variables|TransientData|GetTimeSongStarted)) (Variables|/Script/Engine.SoundBase|GetDuration _sound)) 0.0 (Variables|UserOptions|GetSync%20Player%20Audio))))
    (:else
      (return 0.0))))

(fn IsVisibleInGame
  (return (or (Variables|UserOptions|Basic|GetVisibleInGame) (Utilities|Select :Index (Variables|/Game/Creative/Devices/Common/Components/Creative_VisibleInGame_Component.Creative_VisibleInGame_Component_C|GetForcedVisibilityDuringGame (Variables|Default|GetVisibleInGameComponent)) :NewEnumerator0 false :NewEnumerator1 true :NewEnumerator2 false))))

(fn IsTargetableByLinearMover
  (return true))

(fn MigrateLegacyUserOptions
  (bind _savedtracknamelegacy (Variables|LegacyUserOptions|GetSavedTrackNameLegacy))
  (bind _stringtoaudioconversion (Variables|LegacyUserOptions|GetStringToAudioConversion))
  (bind (_value _returnvalue) (Utilities|Map|Find _stringtoaudioconversion _savedtracknamelegacy))
  (bind (_audio _volume _audiodistance) (Class|GetClassDefaults "/CreativeCoreDevices/Device_Radio_V2.Device_Radio_V2_C"))
  (bind (_audio_2 _volume_2 _playbackspeed _sync_player_audio _restart_audio_on_activation _play _stop _savedtracknamelegacy_2) (Class|GetClassDefaults "/CreativeCoreDevices/Device_Radio_V2.Device_Radio_V2_C"))
  (bind (_audio_3 _volume_3 _playbackspeed_2 _sync_player_audio_2 _restart_audio_on_activation_2 _play_2 _stop_2 _savedtracknamelegacy_3) (Class|GetClassDefaults "/CreativeCoreDevices/Device_Radio_V2.Device_Radio_V2_C"))
  (|CallParentFunction)
  (UserOptions|PlaybackTypeUpgradePath)
  (Utilities|FlowControl|Sequence
    (:then_0
      (Utilities|FlowControl|Sequence
        (:then_0
          ;[1 1 1] All V1 values are > 0, any V1 value needs a conversion applied, then the value is set to 0.
          (if (> (Variables|LegacyUserOptions|GetRadioAttenuationDistance) 0)
            (if (== (Variables|LegacyUserOptions|GetRadioAttenuationDistance) (LocalVariables|GetLegacyInfiniteRange))
              (Variables|UserOptions|SetAudioDistance _audiodistance)
              (Variables|UserOptions|SetLimitAudioDistance false)
              (else
                (Variables|UserOptions|SetAudioDistance (* (Variables|LegacyUserOptions|GetRadioAttenuationDistance) 5.12))))
            (Variables|LegacyUserOptions|SetRadioAttenuationDistance 0.0)
            (LocalVariables|SetbMigratedProperties true)))
        (:then_1
          (if (Utilities|String|NotEqualExactly(String) (Variables|LegacyUserOptions|GetSavedTrackNameLegacy) _savedtracknamelegacy_3)
            (if _returnvalue
              (Variables|UserOptions|Basic|SetAudio _value)
              (Variables|LegacyUserOptions|SetSavedTrackNameLegacy _savedtracknamelegacy_2)
              (LocalVariables|SetbMigratedProperties true))))))
    (:then_1
      (if (LocalVariables|GetbMigratedProperties)
        (OptionData|RefreshPropertyOverrides (Variables|Default|GetToyOptionsComponent))))))

(fn OnRep_Instigators
  ;[0 1 0.1927] Call when when instigators get added or removed.
  (Client|AudioPlaying|ClientUpdateSongState false))

(fn OnRep_TimeSongStarted
  (Client|AudioPlaying|ClientUpdateSongState true))

(fn OnVisiblityChanged
  (params (Visible :type Boolean))
  (if (and Visible (CallFunction|IsinPlayMode))
    (CallFunction|UpdateMeshAndMaterial)))

(fn PlaySoundAtDevice
  ;[0 1 0.9331 0.25] Only Fade in Sound if it's not already playing/fading in.
  (if (not (or (Utilities|Enum|Equal(Enum) (Audio|Components|Audio|GetPlayState (Variables|Default|GetAudio%20Component)) "Playing") (Utilities|Enum|Equal(Enum) (Audio|Components|Audio|GetPlayState (Variables|Default|GetAudio%20Component)) "FadingIn")))
    (Audio|Components|Audio|FadeIn (Variables|Default|GetAudio%20Component) (Variables|UserOptions|GetFadeInDuration) 1.0 (Client|AudioPlaying|GetStartTime) "Linear" :!class /Script/Engine.AudioComponent)))

(fn PlaySoundonPlayer
  (params (PlayerState :type Player%20State%20Object%20Reference))
  (bind (_value _returnvalue) (Utilities|Map|Find (Variables|TransientData|PlayerTracking|GetInstigatorSounds) (Variables|GetPlayerState)))
  (bind _playerstate (Variables|GetPlayerState))
  (bind _fading_out_sounds (Variables|TransientData|PlayerTracking|GetFading%20Out%20Sounds))
  (bind (_value_2 _returnvalue_2) (Utilities|Map|Find _fading_out_sounds _playerstate))
  (Utilities|IsValid PlayerState
    (:Is%20Valid
      (Utilities|FlowControl|Sequence
        (:then_0
          (if _returnvalue_2
            ;[0 1 0.9331 0.25] Reuse an existing component if it exists.
            (LocalVariables|SetLocalAudioComponentReference _value_2) ; User already has a fading out component, \nreuse that rather than re-create.
            (Utilities|Map|Remove (Variables|TransientData|PlayerTracking|GetFading%20Out%20Sounds) (Variables|GetPlayerState))
            (elif (not _returnvalue)
              ;[0.00038 1 0.0591 0.25] Not found a component, create a new one.
              (bind _pawnprivate (Variables|/Script/Engine.PlayerState|GetPawnPrivate (Variables|GetPlayerState))
                (:then
                  (bind _returnvalue_3 (Audio|SpawnSoundAttached :Sound (Variables|/Script/Engine.AudioComponent|GetSound (Variables|Default|GetAudio%20Component)) :AttachToComponent (Variables|Transformation|GetRootComponent _pawnprivate) :AttachPointName "None" :Location (Transformation|GetWorldLocation (Variables|Transformation|GetRootComponent _pawnprivate)) :LocationType "SnapToTarget" :bStopWhenAttachedToDestroyed true :VolumeMultiplier (Variables|UserOptions|Basic|GetVolume) :PitchMultiplier (Variables|UserOptions|GetPlaybackSpeed) :StartTime (Client|AudioPlaying|GetStartTime) :AttenuationSettings (Variables|/Script/Engine.AudioComponent|GetAttenuationSettings (Variables|Default|GetAudio%20Component)) :bAutoDestroy true))
                  (LocalVariables|SetLocalAudioComponentReference _returnvalue_3))
                (:else)))))
        (:then_1
          ;[0 1 0.9331 0.25] Play audio from our component, add it to the tracked audio, and add it to our list of tracked components.
          (bind _localaudiocomponentreference (LocalVariables|GetLocalAudioComponentReference)
            (:then
              (AudioToMPC|AddTrackedAudioComponent (Variables|Default|GetCreative_Radio_AudioToMPC) _localaudiocomponentreference) ; Add the spawned component to the AudioToMPC component \nso the audio played by this radio can be visualized when attached to the player
              (Audio|Components|Audio|FadeIn (LocalVariables|GetLocalAudioComponentReference) (Variables|UserOptions|GetFadeInDuration) 1.0 (Client|AudioPlaying|GetStartTime) "Linear" :!class /Script/Engine.AudioComponent)
              (Utilities|Map|Add (Variables|TransientData|PlayerTracking|GetInstigatorSounds) (Variables|GetPlayerState) (LocalVariables|GetLocalAudioComponentReference)))
            (:else)))))))

(fn PlaybackTypeUpgradePath
  (bind _toyoptionscomponent (Variables|Default|GetToyOptionsComponent))
  (bind _currentoptions (OptionData|GetPropertyOverrides _toyoptionscomponent))
  (bind (_value _returnvalue) (Utilities|Map|Find _currentoptions "PlaybackType"))
  (bind _toyoptionscomponent_2 (Variables|Default|GetToyOptionsComponent))
  (if _returnvalue
    (OptionData|SetPropertyOverrides _toyoptionscomponent_2 (Utilities|Map|MakeMap "PlayDuringPregame" (select (Utilities|String|Equal,CaseInsensitive(String) _value "1") "Yes" "No") "PlayDuringWarmup" (select (Utilities|String|Equal,CaseInsensitive(String) _value "1") "Yes" "No") "PlayDuringGameplay" (select (Utilities|String|Equal,CaseInsensitive(String) _value "1") "Yes" "No") "PlayDuringRoundEnd" (select (Utilities|String|Equal,CaseInsensitive(String) _value "1") "Yes" "No") "PlayDuringGameEnd" (select (Utilities|String|Equal,CaseInsensitive(String) _value "1") "Yes" "No")))
    (OptionData|RemovePropertyOverrides _toyoptionscomponent_2 (Utilities|Array|MakeArray "PlaybackType"))))

(fn ResetDevice
  (Creative|Networking|Authority/Cosmetic
    (:Authority
      (Variables|TransientData|PlayerTracking|SetInstigators (Utilities|Array|MakeArray)) ; Clear Instigators
      (Utilities|FlowControl|Sequence
        (:then_0
          (bind _fortminigameprogress (Variables|Default|GetFortMinigameProgress)
            (:then
              (bind _currentminigame (Variables|/Script/FortniteGame.FortMinigameLogicComponent|GetCurrentMinigame _fortminigameprogress)
                (:then
                  (LocalVariables|SetCurrentState (Variables|/Script/FortniteGame.FortMinigame|GetCurrentState _currentminigame)))
                (:else)))
            (:else)))
        (:then_1
          ;[0.4765 1 0 0.25] If the audio wants to play, play it now (and reset). If we've not got a minigame, we'll be in the pregame state.
          (bind _wants_to_play (Utilities|IsAudioPhaseEnabled (LocalVariables|GetCurrentState))
            (:Update%20State
              (Server|SetAudioState _wants_to_play true))))))))

(fn SetAudioState
  (params (Play :type Boolean) (Restart%20if%20Playing :type Boolean))
  ;[0 1 0.9331 0.25] Server - Starts or Stops the Audio from Playing. Will optionally restart if you're already playing and want to play.
  (Networking|SwitchHasAuthority
    (:Authority
      (if Play
        (if (or Restart%20if%20Playing (< (Variables|TransientData|GetTimeSongStarted) 0))
          (Variables|TransientData|SetTimeSongStarted (Fortnite|Sync|Time|GetServerWorldTimeSeconds)))
        (else
          (Variables|TransientData|SetTimeSongStarted -1.0))))))

(fn SetVisualizationState
  (Networking|CanExecuteCosmeticEvents
    (:True
      (Components|Activation|SetActive (Variables|Default|GetCreative_Radio_AudioToMPC) (Variables|UserOptions|GetVisualization) false))))

(fn SetAttenuationDistance
  (bind _audiodistance (Variables|UserOptions|GetAudioDistance))
  (bind _returnvalue (* _audiodistance 1e+02))
  (bind _cachedattenuationsetting (Variables|TransientData|GetCachedAttenuationSetting))
  (bind _limitaudiodistance (Variables|UserOptions|GetLimitAudioDistance))
  (if (or (!= _returnvalue (Variables|TransientData|GetCachedAttenuationSetting)) (not (Variables|UserOptions|GetLimitAudioDistance)))
    (Variables|TransientData|SetCachedAttenuationSetting (Math|Float|SelectFloat _returnvalue (Variables|Config|GetMaxAttenuationDistance) (Variables|UserOptions|GetLimitAudioDistance)))
    (bind _structout (Utilities|Struct|SetmembersinSoundAttenuationSettings (Variables|/Script/Engine.AudioComponent|GetAttenuationOverrides (Variables|Default|GetAudio%20Component)) _limitaudiodistance _cachedattenuationsetting))
    (Audio|Components|Audio|AdjustAttenuation (Variables|Default|GetAudio%20Component) _structout)
    (bind _values (Utilities|Map|Values (Variables|TransientData|PlayerTracking|GetInstigatorSounds)))
    (for _array_element _values
      (bind _structout_2 (Utilities|Struct|SetmembersinSoundAttenuationSettings (Variables|/Script/Engine.AudioComponent|GetAttenuationOverrides _array_element) _limitaudiodistance _cachedattenuationsetting))
      (Audio|Components|Audio|AdjustAttenuation _array_element _structout_2))))

(fn ShouldInvisiblePropsBeAllowed
  (return false))

(fn StopSounds
  (bind _fadeoutduration (Variables|UserOptions|GetFadeOutDuration))
  (bind _instigatorsounds (Variables|TransientData|PlayerTracking|GetInstigatorSounds))
  ;[1 0.3072 0 0.2] Stops all sounds playing from both the device and any sounds we've spawned attached to players.
  (Audio|Components|Audio|FadeOut (Variables|Default|GetAudio%20Component) (Variables|UserOptions|GetFadeOutDuration) 0.0 "Linear" :!class /Script/Engine.AudioComponent)
  (bind _values (Utilities|Map|Values (Variables|TransientData|PlayerTracking|GetInstigatorSounds)))
  (bind (_arrayindex _arrayelement) (Utilities|Array|ReverseforEachLoop _values)
    (:LoopBody
      (Utilities|IsValid _arrayelement
        (:Is%20Valid
          (Audio|Components|Audio|FadeOut _arrayelement (Variables|UserOptions|GetFadeOutDuration) 0.0 "Linear" :!class /Script/Engine.AudioComponent) ; Fade out sounds.
        )))
    (:Completed
      (Variables|TransientData|PlayerTracking|SetFading%20Out%20Sounds _instigatorsounds)
      (Utilities|Map|Clear _instigatorsounds)
      (if (> _fadeoutduration 0)
        (bind _returnvalue (Utilities|Time|SetTimerbyEvent (EventDispatchers|CreateEvent :!event Destroy%20Fading%20Player%20Sound%20Components) _fadeoutduration)) ; If we're fading out the sound, we don't want to destroy immediately.\nInstead, destroy once fade out time has completed.
        (Variables|TransientData|SetComponent%20Destruction%20Timer%20Handle _returnvalue)
        (else
          (DeviceResetting|DestroyFadingPlayerSoundComponents))))))

(fn UpdateRegisteredPlayers
  (params (Add :type Boolean) (Instigator :type Controller%20Object%20Reference))
  (bind _instigators (Variables|TransientData|PlayerTracking|GetInstigators))
  (bind _returnvalue (Utilities|Array|Length _instigators))
  (bind _returnvalue_2 (Math|Conversions|ToBoolean(Integer) _returnvalue))
  (bind _instigators_2 (Variables|TransientData|PlayerTracking|GetInstigators))
  (Utilities|IsValid Instigator
    (:Is%20Valid
      (bind _output_get (LocalVariables|SetLocal%20PlayerState (Variables|/Script/Engine.Controller|GetPlayerState Instigator)))
      (Utilities|IsValid _output_get
        (:Is%20Valid
          (Utilities|Set|AddItems (LocalVariables|GetLocal%20Instigator%20Cache) _instigators_2)
          (LocalVariables|SetHadInstigators (Math|Conversions|ToBoolean(Integer) (Utilities|Array|Length _instigators_2)))
          (if (Variables|GetAdd)
            (if (not (Utilities|Set|ContainsItem (LocalVariables|GetLocal%20Instigator%20Cache) (LocalVariables|GetLocal%20PlayerState)))
              (Utilities|Set|Add (LocalVariables|GetLocal%20Instigator%20Cache) (LocalVariables|GetLocal%20PlayerState)))
            (else
              (bind _returnvalue_3 (Utilities|Set|Remove (LocalVariables|GetLocal%20Instigator%20Cache) (LocalVariables|GetLocal%20PlayerState)))
              (if _returnvalue_3
                _)))
          (bind _result (Utilities|Set|ToArray (LocalVariables|GetLocal%20Instigator%20Cache)))
          (Variables|TransientData|PlayerTracking|SetInstigators _result)
          (if (and (Variables|UserOptions|GetUpdateStateWhenPlayersRegistered) (!= (LocalVariables|GetHadInstigators) _returnvalue_2))
            (Server|SetAudioState _returnvalue_2 (Variables|UserOptions|GetRestart%20Audio%20On%20Activation))))))))

(fn UpdateVisuals
  (bind _staticmeshcomponent (Variables|Debug|GetStaticMeshComponent))
  (Networking|CanExecuteCosmeticEvents
    (:True
      (if (Editor|InEditor)
        ;[1 0 0.7961] Hiding in Game if we're the editor and don't want to show in game.
        (Development|SetHiddeninGame (Variables|Debug|GetStaticMeshComponent) (not (ChildFunctions|IsVisibleInGame)) false))
      (Variables|/Script/Engine.StaticMeshComponent|GetStaticMesh _staticmeshcomponent
        (:then
          (bind _output_get (LocalVariables|Setlocal%20MID (Material|GetMID _staticmeshcomponent)))
          (Rendering|Material|SetScalarParameterValue _output_get "Pulse Scale" (Math|Float|MapRangeClamped (Variables|UserOptions|Basic|GetVolume) 0.0 4.0 0.0 0.6) :!class /Script/Engine.MaterialInstanceDynamic)
          (Rendering|Material|SetScalarParameterValue _output_get "Pulse Speed" (Variables|UserOptions|GetPlaybackSpeed) :!class /Script/Engine.MaterialInstanceDynamic)
          (Utilities|ShouldPlaySong
            (:Yes
              (Rendering|Material|SetScalarParameterValue _output_get "Turn On" 1.0 :!class /Script/Engine.MaterialInstanceDynamic))
            (:No
              (Rendering|Material|SetScalarParameterValue _output_get "Turn On" 0.0 :!class /Script/Engine.MaterialInstanceDynamic))))
        (:else)))))

(fn UpdateCachedAttenuationSettings)

(fn ConstructionScript
  (Visibility|BindEventtoOnVisibilityChanged (Variables|Default|GetVisibleInGameComponent) (EventDispatchers|CreateEvent :!event OnVisiblityChanged)))
