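; Creative random number generator device that rolls an integer between two limits when overlapped, activated, hit by a music sequencer or on a game phase.
; The server validates team and enabled state, then picks the result after the roll time, optionally drawing from a pick-once array that clients mirror.
; RepNotifies on start time and result drive client cosmetics (zone trigger hops, readout, sounds) and server score, win/lose events and zone activation.
; Instant rolls without reset delay queue extra requests.
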(event EventInitializeDevice
  ;[0 1 0.2378 0.2] Initialization
  (|CallParentFunction)
  (CallFunction|UpdateVisibilityinGame0 "NewEnumerator0")
  (CallFunction|UpdateVolume)
  (if (Editor|InEditor)
    (UEFNConversionHelper|SetComponentHiddenEditor (Variables|Default|GetMovingTrigger) true))
  (Variables|Default|SetEnabled (Class|CreativeCommonDeviceFunctionLibrary|IsEnabledPhaseMatch "PreGame" (CallFunction|GetEnabledonPhaseasInt) self))
  (Networking|SwitchHasAuthority
    (:Authority
      (if (PickOnceLogic|UsesPickOnce)
        (if (PickOnceLogic|NumberArrayNotFull)
          (CallFunction|RPCMULTICASTGenerateNewArray) ; Hiding this RPC behind a number of bits\nof server logic to reduce calls.
        )))))

(event Custom|RPC_MULTICAST%20Generate%20New%20Array
  ;[0 1 0.2378 0.2] Initialization
  (PickOnceLogic|GenerateNumberArray))

(event EventBeginPlay
  ;[0 1 0.2378 0.2] Initialization
  (|CallParentFunction)
  (Networking|SwitchHasAuthority
    (:Authority
      (Ability|EquipFortAbilitySet self "/Game/Creative/Devices/RandomNumberGenerator/GAS_Device_RNG.GAS_Device_RNG"))))

(event MinigameStateChangedDelegate(FortMinigameProgress)
  (params (Minigame :type Fort%20Minigame%20Object%20Reference) (NewMinigameState :type EFortMinigameState%20Enum :default "PreGame"))
  ;[0 0.4323 0.3687 0.248] Minigame State Change
  (PickOnceLogic|GenerateNumbersOnPhase NewMinigameState)
  (CallFunction|ResetReadout Minigame NewMinigameState)
  (Networking|SwitchHasAuthority
    (:Authority
      (Utilities|FlowControl|Sequence
        (:then_0
          (Variables|Default|SetEnabled (Class|CreativeCommonDeviceFunctionLibrary|IsEnabledPhaseMatch NewMinigameState (CallFunction|GetEnabledonPhaseasInt) self)))
        (:then_1
          (if (Class|CreativeCommonDeviceFunctionLibrary2|ActivateonPhase Minigame (Variables|UserOptions|GetActivateOnPhase))
            (CallFunction|ServerStartRoll 0)))))))

(event PostMinigameStartedDelegate(FortMinigameProgress)
  ;[0 0.4323 0.3687 0.248] Minigame State Change
  (CreativeDeviceHelper|ActivateonPhase|BranchActivateonGameStart (Variables|UserOptions|GetActivateOnPhase)
    (:True
      (CallFunction|ServerStartRoll 0))))

(event OnComponentBeginOverlap(PadTrigger)
  (params (OverlappedComponent :type Primitive%20Component%20Object%20Reference) (OtherActor :type Actor%20Object%20Reference) (OtherComp :type Primitive%20Component%20Object%20Reference) (OtherBodyIndex :type Integer) (bFromSweep :type Boolean) (SweepResult :type Hit%20Result%20Structure%20%28by%20ref%29))
  (bind (_cantrigger _controller) (Class|CreativeCommonDeviceFunctionLibrary|ActorCanTrigger OtherActor OtherComp))
  ;[1 1 1] Handles Overlaps
  (if (and (not (Components|ComponentHasTag OtherComp "AffectsChainTrigger")) _cantrigger)
    (CallFunction|ServerStartRoll _controller)))

(event OnComponentBeginOverlap(ChainTrigger)
  (params (OverlappedComponent :type Primitive%20Component%20Object%20Reference) (OtherActor :type Actor%20Object%20Reference) (OtherComp :type Primitive%20Component%20Object%20Reference) (OtherBodyIndex :type Integer) (bFromSweep :type Boolean) (SweepResult :type Hit%20Result%20Structure%20%28by%20ref%29))
  ;[1 1 1] Handles Overlaps
  (if (Components|ComponentHasTag OtherComp "AffectsChainTrigger")
    (if (Utilities|DoesObjectImplementInterface OtherActor "/Game/Creative/Devices/Common/Creative_DeviceInterface.Creative_DeviceInterface_C")
      (bind _instigator (CreativeDevice|GetInstigatingController(Message) OtherActor))
      (else
        _))
    (CallFunction|ServerStartRoll _instigator)))

(event Custom|Pulse%20Moving%20Trigger
  (params (Pulse%20Time :type Float%20%28double-precision%29))
  ;[1 0.6061 0] Pulses the Moving Trigger then Hides the Result
  (if (CallFunction|HasZone)
    (Utilities|RecentlyRolled
      (:true
        (Components|Timeline|SetPlayRate (Variables|Device_RandomNumberGenerator_V2|GetTrigger%20Pulse) (/ 1.0 Pulse%20Time))
        (UEFNConversionHelper|SetComponentHiddenEditor (Variables|Default|GetMovingTrigger))
        (bind (_direction _opacity _emissivescale) (Timeline|Trigger%20Pulse PlayFromStart)
          (:Update
            (Effects|SetTriggerOpacity _opacity)
            (Rendering|Material|SetScalarParameterValue (Variables|VFX|GetBaseMeshMID0) "EmissiveScale" _emissivescale :!class /Script/Engine.MaterialInstanceDynamic))
          (:Finished
            (UEFNConversionHelper|SetComponentHiddenEditor (Variables|Default|GetMovingTrigger) true)))))))

(event Custom|Server%20Start%20Roll
  (params (Instigator :type Controller%20Object%20Reference))
  (bind _returnvalue (Pawn|GetControlledPawn Instigator))
  ;[0 0.5572 1] Server Checks to see if it's valid to start a roll before telling others.
  (Utilities|IsValid Instigator
    (:Is%20Valid
      (Utilities|IsValid _returnvalue
        (:Is%20Valid
          (Variables|Actor|SetInstigator _returnvalue)
          (Variables|Default|SetInstigatorAsController Instigator))
        (:Is%20Not%20Valid
          _)))
    (:Is%20Not%20Valid
      _))
  (if (Math|Boolean|ANDBoolean (or (Creative|TeamOption|IsTeamType (Variables|UserOptions|Basic|GetActivating%20Team) "Any") (== (Creative|TeamOption|ConverttoTeam (Variables|UserOptions|Basic|GetActivating%20Team)) (Team|GetActorTeam Instigator))) (Networking|HasAuthority) (Variables|Default|GetEnabled))
    (if (not (Variables|Default|GetUnavailable))
      (if (== (Variables|UserOptions|GetRollTime) 0.0)
        (Variables|Default|SetUnavailable true)
        (CallFunction|DecideResult)
        (else
          (Variables|Default|SetRoll%20Start%20Time (Fortnite|Sync|Time|GetServerWorldTimeSeconds))))
      (elif (and (== (Variables|UserOptions|GetRollTime) 0.0) (== (Variables|UserOptions|GetResetDelay) 0.0))
        (Utilities|Array|Add (Variables|Default|GetQueuedRequests) (Utilities|Struct|MakeRNGRequest Instigator))))))

(event Custom|Cancel%20Event
  (params (Instigator :type Controller%20Object%20Reference))
  ;[0.7957 0.005113 1] Gameplay Functions
  (CallFunction|StopRoll))

(event Custom|Activate%20Event
  (params (Instigator :type Controller%20Object%20Reference))
  ;[0.7957 0.005113 1] Gameplay Functions
  (CallFunction|ServerStartRoll Instigator))

(event Custom|Reset%20Device
  ;[0 0.762 1] Rolling Logic
  (if (Variables|UserOptions|GetReset%20after%20Use)
    (Utilities|FlowControl|RetriggerableDelay (Math|Float|Max(Float) (Variables|UserOptions|GetResetDelay) 0.2))
    (Variables|Default|SetUnavailable (not (Variables|Default|GetEnabled)))
    (elif (Variables|Default|GetEnabled)
      (Variables|Default|SetEnabled false))))

(event Custom|Start%20Roll
  (params (Seed :type Integer))
  (bind _returnvalue (Fortnite|Sync|Time|GetServerWorldTimeSeconds))
  (bind _rolltime (Variables|UserOptions|GetRollTime))
  (bind _roll_start_time (Variables|Default|GetRoll%20Start%20Time))
  (bind _returnvalue_2 (+ _roll_start_time _rolltime))
  (bind _returnvalue_3 (- _returnvalue_2 _returnvalue))
  (bind _rolldelaycurve (Variables|Default|GetRollDelayCurve))
  (bind _returnvalue_4 (Math|Curves|GetFloatValue _rolldelaycurve _returnvalue_3))
  ;[0 0.762 1] Rolling Logic
  (Networking|SwitchHasAuthority
    (:Authority
      (Variables|Default|SetUnavailable true)
      (Variables|Default|SetResult -1) ; Setting Result to an invalid number\nto ensure it repnotifies properly at end.
      (Utilities|FlowControl|Sequence
        (:then_0
          _)
        (:then_1
          (Utilities|FlowControl|RetriggerableDelay (Variables|UserOptions|GetRollTime))))
      (Utilities|FlowControl|Gate true
        (:Exit
          (CallFunction|DecideResult))))
    (:Remote
      (Math|Random|SetRandomStreamSeed (Variables|Default|GetRandom%20Stream) Seed)
      (if (VisibleInGameComponent|GetDesiredVisibility)
        (Effects|SetTriggerOpacity (Variables|Default|GetRolling%20Trigger%20Opacity%20%25))
        (CallFunction|StartRollClient)
        (else
          _))
      (Utilities|FlowControl|Sequence
        (:then_0
          _)
        (:then_1
          _))
      ;;[0 0.3855 1] Move the Location randomly.
      (Utilities|FlowControl|Gate true
        (:Exit
          (Effects|PlayMoveSFX)
          (if (VisibleInGameComponent|GetDesiredVisibility)
            (CallFunction|SetNewLocation -1) ; -1= Random within Zone.
            (if (< _returnvalue_4 _returnvalue_3)
              (Utilities|FlowControl|RetriggerableDelay _returnvalue_4))))))))

(event Custom|Stop%20Roll%20Server
  ;[0 0.762 1] Rolling Logic
  (Variables|Default|SetUnavailable false)
  (Utilities|FlowControl|Gate true
    (:Exit
      (CallFunction|DecideResult))))

(event Custom|Disable%20Event
  (params (Instigator :type Controller%20Object%20Reference))
  ;[0.7957 0.005113 1] Gameplay Functions
  (if (Variables|Default|GetEnabled)
    (Variables|Default|SetEnabled false)))

(event Custom|Enable%20Event
  (params (Instigator :type Controller%20Object%20Reference))
  ;[0.7957 0.005113 1] Gameplay Functions
  (if (not (Variables|Default|GetEnabled))
    (Variables|Default|SetEnabled true)))

(event Custom|Stop%20Roll
  ;[0 0.762 1] Rolling Logic
  (Variables|Default|SetRoll%20Start%20Time -1.0) ; Set Roll time to a non-existant time.\nThis will cause the roll to stop.
)

(event Custom|Stop%20Roll%20Client
  (bind _returnvalue (Fortnite|Sync|Time|GetServerWorldTimeSeconds))
  (bind _rolltime (Variables|UserOptions|GetRollTime))
  (bind _roll_start_time (Variables|Default|GetRoll%20Start%20Time))
  (bind _returnvalue_2 (+ _roll_start_time _rolltime))
  (bind _returnvalue_3 (- _returnvalue_2 _returnvalue))
  (bind _rolldelaycurve (Variables|Default|GetRollDelayCurve))
  (bind _returnvalue_4 (Math|Curves|GetFloatValue _rolldelaycurve _returnvalue_3))
  ;[0 0.762 1] Rolling Logic
  (Components|Timeline|Stop (Variables|Device_RandomNumberGenerator_V2|GetTrigger%20Pulse))
  (Effects|HideRollVisuals)
  ;;[0 0.3855 1] Move the Location randomly.
  (Utilities|FlowControl|Gate true
    (:Exit
      (Effects|PlayMoveSFX)
      (if (VisibleInGameComponent|GetDesiredVisibility)
        (CallFunction|SetNewLocation -1) ; -1= Random within Zone.
        (if (< _returnvalue_4 _returnvalue_3)
          (Utilities|FlowControl|RetriggerableDelay _returnvalue_4))))))

(event UserOptions|EventHandleUserOptionsChanged
  (params (ChangedPropertyNames :type Array%20of%20Names))
  (Networking|CanExecuteCosmeticEvents
    (:True
      (CallFunction|UpdateReadoutMaterials))))

(event OnVisibilityChanged(ZoneVisibilityNoCollisionInGameComponent)
  (params (Visible :type Boolean))
  (CallFunction|UpdateVolume))

(event EventOnHitByMusicSequencer
  (params (InstigatedBy :type Controller%20Object%20Reference))
  (CallFunction|ServerStartRoll InstigatedBy))

(event OnVisibilityChanged(VisibleInGameComponent)
  (params (Visible :type Boolean))
  ;[0.000566 0.4011 1 0.1] Visibility Switch - Disable Base Mesh Collision if not visible (other collision stays active to handle overlaps)
  (Collision|SetCollisionEnabled (Variables|Default|GetBase%20Mesh) (select Visible "QueryAndPhysics" "NoCollision")))

(fn ActivateFromOtherDevice
  (params (Instigator :type Controller%20Object%20Reference) (Requester :type ERequesterType%20Enum :default "NewEnumerator0"))
  (return false))

(fn ActivateOverlappingTriggers
  (if (CallFunction|HasZone)
    (Ability|SendGameplayEventtoActor self (Variables|Default|GetActivation%20Tag) (Utilities|Struct|MakeGameplayEventData (Variables|Default|GetActivation%20Tag) (Game|GetInstigatorController)))))

(fn AttemptEmptyQueue
  (bind _queuedrequests (Variables|Default|GetQueuedRequests))
  (Networking|SwitchHasAuthority
    (:Authority
      (if (Utilities|IsArrayNonEmpty (Variables|Default|GetQueuedRequests))
        (bind _output_get (LocalVariables|SetFirst%20Request (Utilities|Array|Get(acopy) _queuedrequests)))
        (Utilities|Array|RemoveIndex _queuedrequests)
        (CallFunction|ServerStartRoll (Utilities|Struct|BreakRNGRequest _output_get))))))

(fn AwardScore
  (bind _returnvalue (Game|GetInstigatorController))
  (bind _result (Variables|Default|GetResult))
  (bind _winningvalue (Variables|UserOptions|Basic|GetWinningValue))
  (switch Utilities|FlowControl|Switch|SwitchonEAwardScoreOn (Variables|UserOptions|GetAwardScore)
    (:NewEnumerator0
      _)
    (:NewEnumerator1
      (if (>= _result _winningvalue)
        _))
    (:NewEnumerator2
      (if (< _result _winningvalue)
        _)))
  (Utilities|IsValid _returnvalue
    (:Is%20Valid
      (Class|CreativeCommonDeviceFunctionLibrary|AwardScore self (select (== (Variables|UserOptions|GetScore%20Value) 0) (Math|Float|Round (* (Variables|UserOptions|GetResultMultiplier) (Variables|Default|GetResult))) (Variables|UserOptions|GetScore%20Value)) (Variables|UserOptions|GetScoreType) _returnvalue (Variables|Default|GetFortMinigameProgress)))))

(fn CanInteract
  (params (InteractingPawn :type Fort%20Pawn%20Object%20Reference) (InteractionBeingAttempted :type EInteractionBeingAttempted%20Enum :default "FirstInteraction") (InteractionType :type TInteractionType%20Enum :default "IT_NoInteraction"))
  (return false))

(fn DebugPrintNumberArray
  (params (Prefix :type String :default "Number Array: "))
  (if (Variables|Default|GetShow%20Debug)
    (LocalVariables|SetDebug%20String Prefix)
    (for _array_element (Variables|Default|GetValid%20Numbers)
      (LocalVariables|SetDebug%20String (Utilities|String|Append (LocalVariables|GetDebug%20String) (Utilities|String|ToString(Integer) _array_element) ", ")))
    (Development|PrintString (LocalVariables|GetDebug%20String) true true "(R=1.000000,G=0.744069,B=0.000000,A=1.000000)" 1e+01)))

(fn DecideResult
  (Math|Random|SeedRandomStream (Variables|Default|GetRandom%20Stream))
  (bind _value (PickOnceLogic|GetRandomRemainingValue (PickOnceLogic|UsesPickOnce)))
  (Variables|Default|SetResult _value)
  (CallFunction|ResetDevice))

(fn GenerateNumberArray
  (bind _valuelimit1 (Variables|UserOptions|Basic|GetValueLimit1))
  (bind _valuelimit2 (Variables|UserOptions|Basic|GetValueLimit2))
  (if (or (Variables|Default|GetHas%20Initially%20Replicated) (Networking|HasAuthority))
    (for _index (range (+ (Math|Integer|Absolute(Integer) (- _valuelimit1 _valuelimit2)) 1))
      (Utilities|Array|Add (LocalVariables|Getlocal%20Array) (+ _index (Math|Integer|Min(Integer) _valuelimit1 _valuelimit2))))
    (Variables|Default|SetValid%20Numbers (LocalVariables|Getlocal%20Array)) ; RepNotify InitialOnly
    (PickOnceLogic|DebugPrintNumberArray "Generated New Array: ")))

(fn GenerateNumbersonPhase
  (params (Current%20Phase :type EFortMinigameState%20Enum :default "PreGame"))
  (if (Utilities|Enum|Equal(Enum) (Variables|GetCurrent%20Phase) "Setup")
    (if (and (Utilities|Enum|Equal(Enum) (Variables|UserOptions|Basic|GetPick%20Each%20Number%20Once) "NewEnumerator1") (Variables|Default|GetHas%20Round%20Ended))
      (Variables|Default|SetHas%20Round%20Ended false)
      (elif (and (PickOnceLogic|UsesPickOnce) (PickOnceLogic|NumberArrayNotFull))
        (PickOnceLogic|GenerateNumberArray)))
    (elif (and (Utilities|Enum|Equal(Enum) (Variables|GetCurrent%20Phase) "PostRoundEnd") (Utilities|Enum|Equal(Enum) (Variables|UserOptions|Basic|GetPick%20Each%20Number%20Once) "NewEnumerator1"))
      (Variables|Default|SetHas%20Round%20Ended true)
      (elif (Utilities|Enum|Equal(Enum) (Variables|GetCurrent%20Phase) "PostGameEnd")
        _))))

(fn GetDeviceColor
  (return 0))

(fn GetInstigatingController
  (return (Variables|Default|GetInstigatorAsController)))

(fn GetRandomRemainingValue
  (params (Remove%20Once%20Chosen :type Boolean))
  (bind _valid_numbers (Variables|Default|GetValid%20Numbers))
  (Utilities|FlowControl|Sequence
    (:then_0
      (if (PickOnceLogic|UsesPickOnce)
        (if (Utilities|IsArrayNonEmpty (Variables|Default|GetValid%20Numbers))
          (bind _output_get (LocalVariables|Set_i (Math|Random|RandomIntegerinRangefromStream (Variables|Default|GetRandom%20Stream) 0 (Utilities|Array|LastIndex (Variables|Default|GetValid%20Numbers)))))
          (LocalVariables|Setchosen%20value (Utilities|Array|Get(acopy) (Variables|Default|GetValid%20Numbers) _output_get))
          (if (Variables|GetRemove%20Once%20Chosen)
            (Utilities|Array|RemoveIndex _valid_numbers (LocalVariables|Get_i))
            (PickOnceLogic|DebugPrintNumberArray (Utilities|String|Append "Removed " (Utilities|String|ToString(Integer) (LocalVariables|Getchosen%20value)) ": "))
            (if (not (Utilities|IsArrayNonEmpty _valid_numbers))
              (PickOnceLogic|GenerateNumberArray))))
        (else
          ;[1 1 1] Ignore all of the Array stuff if we're not using the Pick Once system.
          (LocalVariables|Setchosen%20value (Math|Random|RandomIntegerinRangefromStream (Variables|Default|GetRandom%20Stream) (Math|Integer|Min(Integer) (Variables|UserOptions|Basic|GetValueLimit1) (Variables|UserOptions|Basic|GetValueLimit2)) (Math|Integer|Max(Integer) (Variables|UserOptions|Basic|GetValueLimit1) (Variables|UserOptions|Basic|GetValueLimit2)))))))
    (:then_1
      (return (LocalVariables|Getchosen%20value)))))

(fn GetCreativeActorBounds
  (bind (_origin _boxextent) (Collision|GetComponentBounds (Variables|Default|GetBase%20Mesh)))
  (return true _boxextent))

(fn GetCreativeActorOrigin
  (return true (Variables|/Script/Engine.SceneComponent|GetRelativeLocation (Variables|Default|GetAnchor))))

(fn GetCreativeActorRelativeTransform
  (return false 0))

(fn GetEnabledOnPhaseAsInt
  ;[0 0.4411 1] Converting our EnabledDuringPhase (0 index to -1 index, just a -1)
  (return (- (Math|Conversions|ToInteger(Byte) (Variables|UserOptions|GetEnabledDuringPhase)) 1)))

(fn GetGridAxisGuideAttachmentPoint
  (return true true "0.000000,256.000000,0.000000"))

(fn GetMusicSequencerHitSound
  (params (InstigatedBy :type Controller%20Object%20Reference))
  (return 0 0 0 0))

(fn GetMusicSequencerHitSoundLocation
  (params (InstigatedBy :type Controller%20Object%20Reference))
  (return "0, 0, 0"))

(fn GetNumberOfSubdivisions
  (return (+ (Math|Integer|Absolute(Integer) (- (Math|Integer|Max(Integer) (Variables|UserOptions|Basic|GetValueLimit1) (Variables|UserOptions|Basic|GetValueLimit2)) (Math|Integer|Min(Integer) (Variables|UserOptions|Basic|GetValueLimit1) (Variables|UserOptions|Basic|GetValueLimit2)))) 1)))

(fn GetOverrideMeshMaterial
  (params (MeshComp :type Static%20Mesh%20Component%20Object%20Reference) (MatIdx :type Integer))
  (Utilities|FlowControl|Sequence
    (:then_0
      (if (and (Utilities|Equal(Object) (Variables|GetMeshComp) (Variables|Default|GetBase%20Mesh)) (== (Variables|GetMatIdx) 0))
        (return (Variables|VFX|GetBaseMeshMID0))))
    (:then_1
      (if (and (Utilities|Equal(Object) (Variables|GetMeshComp) (Variables|Default|GetBase%20Mesh)) (== (Variables|GetMatIdx) 2))
        (return (Variables|VFX|GetBaseMeshMID2))))
    (:then_2
      (if (Utilities|Equal(Object) (Variables|GetMeshComp) (Variables|Default|GetFrame))
        (return (Variables|VFX|GetFrame%20MID))))
    (:then_3
      (return (|CallParentFunction (Variables|GetMeshComp) (Variables|GetMatIdx))))))

(fn GetSpawnedPickups
  (return 0))

(fn HasZone
  (return (Utilities|Enum|NotEqual(Enum) (Variables|UserOptions|GetZoneDirection) "NewEnumerator0")))

(fn HideRollVisuals
  (CallFunction|ReattachMIDsIfNecessary)
  ;[1 1 1] Color Updates to neutral state
  (Rendering|Material|SetScalarParameterValue (Variables|VFX|GetBaseMeshMID0) "Compute" 0.0 :!class /Script/Engine.MaterialInstanceDynamic)
  (Rendering|Material|SetScalarParameterValue (Variables|VFX|GetBaseMeshMID2) "ComputeSpeed" (Variables|VFX|GetIdleSpeed) :!class /Script/Engine.MaterialInstanceDynamic)
  (Rendering|Material|SetVectorParameterValue (Variables|VFX|GetBaseMeshMID2) "Emissive" (Variables|VFX|GetDotsIdleColor) :!class /Script/Engine.MaterialInstanceDynamic)
  (Rendering|Material|SetScalarParameterValue (Variables|VFX|GetBaseMeshMID2) "ComputeEmissiveBias" 0.0 :!class /Script/Engine.MaterialInstanceDynamic)
  (Rendering|Material|SetVectorParameterValue (Variables|VFX|GetBaseMeshMID0) "Emissive" (Variables|VFX|GetLCDIdleColor) :!class /Script/Engine.MaterialInstanceDynamic)
  (Rendering|Material|SetVectorParameterValue (Variables|VFX|GetBaseMeshMID0) "BaseColor" (Variables|VFX|GetLCDIdleBaseColor) :!class /Script/Engine.MaterialInstanceDynamic))

(fn IsEnabled
  (return false))

(fn IsHoldingItemForPawn
  (params (Item :type Item%20Variant%20Handle%20Structure) (Pawn :type Pawn%20Object%20Reference))
  (return false))

(fn IsVisibleDuringMinigame
  (return (Variables|UserOptions|GetVisible%20During%20Game)))

(fn MigrateLegacyUserOptions
  (|CallParentFunction)
  (bind (_found _value) (Creative|ToyOptions|GetOptionValueasInt (Variables|Default|GetToyOptionsComponent) "ResetDelay"))
  (if (and _found (< _value 0))
    (Variables|UserOptions|SetReset%20after%20Use false) ; If we have a reset delay of less than 0, \nwe never want to reset. 
    (Variables|UserOptions|SetResetDelay 0.0)
    (OptionData|RefreshPropertyOverrides (Variables|Default|GetToyOptionsComponent))))

(fn NumberArrayNotFull
  (return (!= (Utilities|Array|LastIndex (Variables|Default|GetValid%20Numbers)) (Math|Integer|Absolute(Integer) (- (Variables|UserOptions|Basic|GetValueLimit1) (Variables|UserOptions|Basic|GetValueLimit2))))))

(fn OnRep_Enabled
  (Networking|SwitchHasAuthority
    (:Authority
      (bind _output_get (Variables|Default|SetUnavailable (not (Variables|Default|GetEnabled)))) ; Becomes available when enabled,\neven if it was in a shutdown state when it was disabled.
      (if _output_get
        (CallFunction|StopRoll)
        (Utilities|Array|Clear (Variables|Default|GetQueuedRequests))))
    (:Remote
      (Rendering|Material|SetScalarParameterValue (Variables|VFX|GetBaseMeshMID0) "EmissiveScale" (Math|Conversions|ToFloat(Boolean) (Variables|Default|GetEnabled)) :!class /Script/Engine.MaterialInstanceDynamic)
      (if (not (Variables|Default|GetEnabled))
        ;[1 1 1] Disable Effects
        (Components|Timeline|Stop (Variables|Device_RandomNumberGenerator_V2|GetTrigger%20Pulse))
        (Effects|SetTriggerOpacity 0.0)))))

(fn OnRep_Result
  (bind _returnvalue (Game|GetInstigatorController))
  (bind _valuelimit2 (Variables|UserOptions|Basic|GetValueLimit2))
  (bind _result (Variables|Default|GetResult))
  (bind _returnvalue_2 (== _result _valuelimit2))
  (bind _valuelimit1 (Variables|UserOptions|Basic|GetValueLimit1))
  (bind _returnvalue_3 (== _valuelimit1 _result))
  (bind _returnvalue_4 (Game|GetInstigatorController))
  (bind _returnvalue_5 (or _returnvalue_3 _returnvalue_2))
  (if (>= (Variables|Default|GetResult) 0)
    (CallFunction|SetNewLocation (Variables|Default|GetResult))
    (Networking|SwitchHasAuthority
      (:Authority
        (CallFunction|ActivateOverlappingTriggers)
        (CallFunction|AwardScore)
        (if (>= (Variables|Default|GetResult) (Variables|UserOptions|Basic|GetWinningValue))
          (GameplayEvent|BroadcastEvent (Variables|UserOptions|Events|GetOn%20Win) 0 (Utilities|Struct|MakeGameplayEventWinRNG _returnvalue_4))
          (if _returnvalue_5
            (GameplayEvent|BroadcastEvent (Variables|UserOptions|Events|GetOn%20Rolled%20Max) 0 (Utilities|Struct|MakeGameplayEventMaxRNG _returnvalue)))
          (else
            (GameplayEvent|BroadcastEvent (Variables|UserOptions|Events|GetOn%20Lose) 0 (Utilities|Struct|MakeGameplayEventLoseRNG _returnvalue_4))
            (if _returnvalue_5
              (GameplayEvent|BroadcastEvent (Variables|UserOptions|Events|GetOn%20Rolled%20Min) 0 (Utilities|Struct|MakeGameplayEventMinRNG _returnvalue))))))
      (:Remote
        (CallFunction|StopRollClient)
        (if (VisibleInGameComponent|GetDesiredVisibility)
          (CallFunction|UpdateReadout (Math|Float|Round (* (Variables|Default|GetResult) (Variables|UserOptions|GetResultMultiplier))))
          (CallFunction|PulseMovingTrigger (Variables|Effects|GetFinal%20Pulse%20Time))
          (else
            _))
        (Effects|PlayEndFX)
        (PickOnceLogic|RemoveNumberFromArray (Variables|Default|GetResult))))))

(fn OnRep_RollStartTime
  (if (> (Variables|Default|GetRoll%20Start%20Time) 0.0)
    (if (< (Fortnite|Sync|Time|GetServerWorldTimeSeconds) (+ (Variables|Default|GetRoll%20Start%20Time) (Variables|UserOptions|GetRollTime)))
      (CallFunction|StartRoll (Math|Float|Truncate (Variables|Default|GetRoll%20Start%20Time))))
    (else
      (Networking|SwitchHasAuthority
        (:Authority
          (CallFunction|StopRollServer))
        (:Remote
          (CallFunction|StopRollClient))))))

(fn OnRep_Unavailable
  (if (not (Variables|Default|GetUnavailable))
    (Networking|SwitchHasAuthority
      (:Authority
        (CallFunction|AttemptEmptyQueue))
      (:Remote
        (Networking|CanExecuteCosmeticEvents
          (:True
            (CallFunction|SetFlashingLights true)))))))

(fn OnRep_ValidNumbers
  (Networking|SwitchHasAuthority
    (:Remote
      (Variables|Default|SetHas%20Initially%20Replicated true) ; This has COND_InitialOnly to set initial state, \nthen doesn't replicate anything else.\nAs such flag this once we have valid numbers, \nto let the client know it can start creating for itself now.
      (PickOnceLogic|DebugPrintNumberArray "Replicated Array: "))))

(fn PlayEndFX
  (Utilities|RecentlyRolled
    (:true
      (if (VisibleInGameComponent|GetDesiredVisibility)
        (CallFunction|SetFlashingLights false) ; Disable flashing lights \n(We'll re-enable when it becomes available)
        (else
          _))
      (if (Variables|UserOptions|GetPlay%20Audio)
        (Audio|PlaySoundatLocation (select (>= (Variables|Default|GetResult) (Variables|UserOptions|Basic|GetWinningValue)) (Variables|Effects|GetWin%20Sound) (Variables|Effects|GetLose%20Sound)) (Transformation|GetWorldLocation (Variables|Default|GetResultVisualBase)))))))

(fn PlayMoveSFX
  (if (Variables|UserOptions|GetPlay%20Audio)
    (Audio|PlaySoundatLocation (Variables|Effects|GetMove%20Trigger%20Sound) (Transformation|GetWorldLocation (Variables|Default|GetResultVisualBase)))))

(fn ReattachMIDsifNecessary
  (bind _basemeshmid0 (Variables|VFX|GetBaseMeshMID0))
  (bind _basemeshmid2 (Variables|VFX|GetBaseMeshMID2))
  (bind _frame_mid (Variables|VFX|GetFrame%20MID))
  (if (Utilities|NotEqual(Object) _frame_mid (Rendering|Material|GetMaterial (Variables|Default|GetFrame) 0))
    (Rendering|Material|SetMaterial (Variables|Default|GetFrame) 0 _frame_mid)
    (else
      _))
  (for _index (range 2)
    (if (Utilities|NotEqual(Object) (select (!= _index 0) _basemeshmid2 _basemeshmid0) (Rendering|Material|GetMaterial (Variables|Default|GetBase%20Mesh) (* _index 2)))
      (Rendering|Material|SetMaterial (Variables|Default|GetBase%20Mesh) (* _index 2) (select (!= _index 0) _basemeshmid2 _basemeshmid0)))))

(fn RemoveNumberfromArray
  (params (Number :type Integer))
  (bind _valid_numbers (Variables|Default|GetValid%20Numbers))
  (if (PickOnceLogic|UsesPickOnce)
    (Utilities|Array|RemoveItem _valid_numbers Number)
    (PickOnceLogic|DebugPrintNumberArray (Utilities|String|Append "Removed " (Utilities|String|ToString(Integer) Number) ": "))
    (if (not (Utilities|IsArrayNonEmpty _valid_numbers))
      (PickOnceLogic|GenerateNumberArray))))

(fn ResetReadout
  (params (Minigame :type Fort%20Minigame%20Object%20Reference) (New%20State :type EFortMinigameState%20Enum :default "PreGame"))
  (if (or (and (or (== (Variables|/Script/FortniteGame.FortMinigame|GetCurrentRound Minigame) 1) (Utilities|Enum|NotEqual(Enum) (Variables|UserOptions|Basic|GetPick%20Each%20Number%20Once) "NewEnumerator1")) (Utilities|Enum|Equal(Enum) New%20State "Transitioning")) (Utilities|Enum|Equal(Enum) New%20State "PostGameEnd"))
    (CallFunction|UpdateReadout 0)))

(fn SetFlashingLights
  (params (On :type Boolean))
  (CallFunction|ReattachMIDsIfNecessary)
  (Rendering|Material|SetScalarParameterValue (Variables|VFX|GetBaseMeshMID2) "EmissiveScale" (select On 1.0 0.0) :!class /Script/Engine.MaterialInstanceDynamic) ; Disable flashing lights \n(We'll re-enable when it becomes available)
)

(fn SetFloorPadVisuals)

(fn SetNewLocation
  (params (Value :type Integer :default -1))
  (bind _movingtrigger (Variables|Default|GetMovingTrigger))
  (bind (_relativelocation_x _relativelocation_y _relativelocation_z) (Math|Vector|BreakVector (Variables|/Script/Engine.SceneComponent|GetRelativeLocation_X (Variables|Default|GetMovingTrigger))))
  (bind _relativescale3d_x (Math|Vector|BreakVector (Variables|/Script/Engine.SceneComponent|GetRelativeScale3D_X (Variables|Default|GetMovingTrigger))))
  (if (CallFunction|HasZone)
    (bind _value (PickOnceLogic|GetRandomRemainingValue false))
    (LocalVariables|Setlocal%20Index (select (< Value 0) _value Value))
    (Transformation|SetRelativeLocation _movingtrigger (Math|Vector|MakeVector (+ (* (Math|Integer|Max(Integer) (- (LocalVariables|Getlocal%20Index) (Math|Integer|Min(Integer) (Variables|UserOptions|Basic|GetValueLimit1) (Variables|UserOptions|Basic|GetValueLimit2)))) (* _relativescale3d_x 512.0)) (/ (* _relativescale3d_x 512.0) 2.0)) _relativelocation_y _relativelocation_z) false false)
    (UEFNConversionHelper|SetComponentHiddenEditor _movingtrigger)))

(fn SetTriggerCollision
  (params (On :type Boolean))
  (Collision|SetCollisionEnabled (Variables|Default|GetMovingTrigger) (select On "QueryOnly" "NoCollision"))
  (return false))

(fn SetTriggerOpacity
  (params (Value :type Float%20%28double-precision%29))
  (Networking|CanExecuteCosmeticEvents
    (:True
      (Rendering|Material|SetScalarParameterValue (Material|GetMID (Variables|Default|GetMovingTrigger)) "Opacity" Value :!class /Script/Engine.MaterialInstanceDynamic))))

(fn StartRollClient
  (CallFunction|ReattachMIDsIfNecessary)
  (Rendering|Material|SetScalarParameterValue (Variables|VFX|GetBaseMeshMID2) "ComputeEmissiveBias" (Variables|VFX|GetDotsEmissiveBias) :!class /Script/Engine.MaterialInstanceDynamic)
  (Rendering|Material|SetScalarParameterValue (Variables|VFX|GetBaseMeshMID2) "ComputeSpeed" (Variables|VFX|GetComputeSpeed) :!class /Script/Engine.MaterialInstanceDynamic)
  (Rendering|Material|SetVectorParameterValue (Variables|VFX|GetBaseMeshMID2) "Emissive" (Variables|VFX|GetDotsComputingColor) :!class /Script/Engine.MaterialInstanceDynamic)
  (Rendering|Material|SetScalarParameterValue (Variables|VFX|GetBaseMeshMID0) "Compute" 1.0 :!class /Script/Engine.MaterialInstanceDynamic)
  (Rendering|Material|SetVectorParameterValue (Variables|VFX|GetBaseMeshMID0) "Emissive" (Variables|VFX|GetLCDComputingColor) :!class /Script/Engine.MaterialInstanceDynamic)
  (Rendering|Material|SetVectorParameterValue (Variables|VFX|GetBaseMeshMID0) "BaseColor" (Variables|VFX|GetLCDComputingBaseColor) :!class /Script/Engine.MaterialInstanceDynamic)
  (CallFunction|UpdateReadout (Math|Random|RandomIntegerinRange 100 999)))

(fn UpdateReadout
  (params (Number :type Integer))
  (bind _local_digit (LocalVariables|Getlocal%20Digit))
  (bind _returnvalue (Math|Integer|%25%28Integer%29 _local_digit 10))
  (CallFunction|ReattachMIDsIfNecessary)
  (LocalVariables|Setlocal%20Digit Number)
  (for _array_element (Variables|VFX|GetMaterialNumberParameters)
    (Rendering|Material|SetScalarParameterValue (Variables|VFX|GetBaseMeshMID0) _array_element (* (Math|Conversions|ToFloat(Integer) _returnvalue) 0.1) :!class /Script/Engine.MaterialInstanceDynamic)
    (LocalVariables|Setlocal%20Digit (/ (- (LocalVariables|Getlocal%20Digit) _returnvalue) 10))))

(fn UpdateReadoutMaterials
  (bind _returnvalue (Rendering|Material|CreateDynamicMaterialInstance (Variables|Default|GetBase%20Mesh) 0 0 "None" :!class /Script/Engine.PrimitiveComponent))
  (Variables|VFX|SetBaseMeshMID0 _returnvalue))

(fn UpdateVisibilityInGame_0
  (params (New%20Visible%20During%20Phase :type EVisibleInGameState%20Enum :default "NewEnumerator0"))
  (Visibility|SetVisibilityComponents (Variables|Default|GetZoneVisibilityNoCollisionInGameComponent) (Utilities|Array|MakeArray (Variables|Default|GetFrame) (Variables|Default|GetSubdivisionPlane)) false)
  (Visibility|SetVisiblityDuringPhase (Variables|Default|GetZoneVisibilityNoCollisionInGameComponent) (select (CallFunction|HasZone) "NewEnumerator3" "NewEnumerator0")))

(fn UpdateVolume
  (CallFunction|ReattachMIDsIfNecessary)
  (Rendering|Material|SetVectorParameterValue (Variables|VFX|GetFrame%20MID) "SizeBias" (Math|Conversions|ToLinearColor(Vector) (- (* (Math|Vector|MakeVector 256.0 256.0 384.0) (Math|Vector|MakeVector (Variables|UserOptions|GetLength) (Variables|UserOptions|GetWidth) (Variables|UserOptions|GetHeight))) (Math|Vector|MakeVector 256.0 256.0 384.0))) :!class /Script/Engine.MaterialInstanceDynamic)
  (Rendering|SetBoundsScale (Variables|Default|GetFrame) (* (Math|Float|Max(Float) (Variables|UserOptions|GetLength) (Variables|UserOptions|GetWidth) (Variables|UserOptions|GetHeight)) 1.5))
  (Transformation|SetRelativeScale3D (Variables|Default|GetMovingTrigger) (Math|Vector|MakeVector (Effects|ZoneXScaleModified) (Variables|UserOptions|GetWidth) (Variables|UserOptions|GetHeight)))
  (Transformation|SetRelativeLocation (Variables|Default|GetFrameAnchor) (Math|Vector|MakeVector (* (Variables|UserOptions|GetLength) 256.0)) false false)
  (Transformation|SetRelativeScale3D (Variables|Default|GetSubdivisionPlane) (Math|Vector|MakeVector (* (Variables|UserOptions|GetLength) 512.0) (* (Variables|UserOptions|GetWidth) 512.0) 1.0))
  (Rendering|Material|SetScalarParameterValue (Variables|VFX|GetSubdivisionPlaneMID) "Subdivisions" (Math|Conversions|ToFloat(Integer) (CallFunction|GetNumberOfSubdivisions)) :!class /Script/Engine.MaterialInstanceDynamic)
  (Transformation|SetRelativeTransform (Variables|Default|GetTriggerBase) (Utilities|Map|Find (Variables|Default|GetDirectionMap) (Variables|UserOptions|GetZoneDirection)) false true)
  ;[1 1 1] Reset location and readout if we are updating the volume in edit mode.
  (if (not (Class|CreativeCommonDeviceFunctionLibrary|IsinPlayMode self))
    (CallFunction|SetNewLocation (Math|Float|Round (* (Variables|Default|GetResult) (Variables|UserOptions|GetResultMultiplier))))
    (CallFunction|PulseMovingTrigger (Variables|Effects|GetFinal%20Pulse%20Time)) ; Pulse the Volume to show the player it's current size.
    (CallFunction|UpdateReadout (Variables|Default|GetResult)) ; Invalid Numbers (negative) = Off
  ))

(fn ConstructionScript
  (Networking|CanExecuteCosmeticEvents
    (:True
      (bind _returnvalue (Rendering|Material|CreateDynamicMaterialInstance (Variables|Default|GetBase%20Mesh) 0 0 "None" :!class /Script/Engine.PrimitiveComponent))
      (Variables|VFX|SetBaseMeshMID0 _returnvalue)
      (bind _returnvalue_2 (Rendering|Material|CreateDynamicMaterialInstance (Variables|Default|GetBase%20Mesh) 2 0 "None" :!class /Script/Engine.PrimitiveComponent))
      (Variables|VFX|SetBaseMeshMID2 _returnvalue_2)
      (bind _returnvalue_3 (Rendering|Material|CreateDynamicMaterialInstance (Variables|Default|GetSubdivisionPlane) 0 0 "None" :!class /Script/Engine.PrimitiveComponent))
      (Variables|VFX|SetSubdivisionPlaneMID _returnvalue_3)
      (Variables|VFX|SetFrame%20MID (Material|GetMID (Variables|Default|GetFrame))))))

(fn UsesPickOnce
  (return (Utilities|Enum|NotEqual(Enum) (Variables|UserOptions|Basic|GetPick%20Each%20Number%20Once) "NewEnumerator0")))

(fn VFXParameters
  (bind (_returnvalue_x _returnvalue_y _returnvalue_z) (Math|Vector|BreakVector (* (Math|Vector|MakeVector 512.0 512.0 384.0) (Math|Vector|MakeVector (Effects|ZoneXScaleModified) (Variables|UserOptions|GetWidth) (Variables|UserOptions|GetHeight)))))
  (return (Utilities|Array|MakeArray :%5B0%5D (Utilities|Struct|MakeParticleSysParam "BlocksVolume" "PSPT_VectorRand" 0.0 0.0 (* (Variables|Effects|GetBlocksVolumeHigh) (Math|Vector|MakeVector (Effects|ZoneXScaleModified) (Variables|UserOptions|GetWidth) (Variables|UserOptions|GetHeight))) (* (Variables|Effects|GetBlocksVolumeLow) (Math|Vector|MakeVector (Effects|ZoneXScaleModified) (Variables|UserOptions|GetWidth) (Variables|UserOptions|GetHeight)))) :%5B1%5D (Utilities|Struct|MakeParticleSysParam "BlocksSpawnRateScale" "PSPT_Scalar" (Utilities|ClampMaxFloat (/ (Math|Float|float*float _returnvalue_x _returnvalue_y _returnvalue_z) 262144.0) 1e+01) 0.0 "0.000000,0.000000,0.000000" (0.0 0.0 0.0) "(B=0,G=0,R=0,A=0)" 0 0))))

(fn ZoneXScaleModified
  (return (/ (Variables|UserOptions|GetLength) (Math|Conversions|ToFloat(Integer) (+ (Math|Integer|Absolute(Integer) (- (Variables|UserOptions|Basic|GetValueLimit1) (Variables|UserOptions|Basic|GetValueLimit2))) 1)))))
