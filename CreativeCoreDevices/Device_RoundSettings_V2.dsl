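; Round Settings creative device that applies per-round rules to a Fortnite minigame, either for every round or one chosen round.
; Minigame state changes drive it: warmup grants capped resources plus previous-winner bonuses, round start broadcasts and can reset classes, round end can disable join-in-progress.
; Receiver events enable or disable it, control join-in-progress, drop standard end conditions or end the round.
; It also migrates legacy options and receiver bindings.
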(event EventOnWorldReady
  ;[0 1 0.08672] Visible In Game Component needs initialized even when working in basic mode as BP components can't initialize in editor,
  ; and BuildingProp based devices need to initialize visibleingame OnWorldReady
  (|CallParentFunction)
  (Class|CreativeVisibleinGameComponent|InitializeComponent (Variables|Default|GetVisibleInGameComponent)))

(event EditorPreview|EventSetupForEditorPreview
  ;[0 1 0.08672] Visible In Game Component needs initialized even when working in basic mode as BP components can't initialize in editor,
  ; and BuildingProp based devices need to initialize visibleingame OnWorldReady
  (|CallParentFunction)
  (Class|CreativeVisibleinGameComponent|InitializeComponent (Variables|Default|GetVisibleInGameComponent)))

(event EventBeginPlay
  (Networking|SwitchHasAuthority
    (:Authority
      (Variables|RoundSettings|SetbActiveDuringMinigame (Variables|UserOptions|GetbEnableOnMinigameStart)))))

(event OnAnyPropertyChanged(ToyOptionsComponent)
  (Networking|SwitchHasAuthority
    (:Authority
      (Variables|RoundSettings|SetbActiveDuringMinigame (Variables|UserOptions|GetbEnableOnMinigameStart)))))

(event OnPropertiesLoaded(ToyOptionsComponent)
  (Networking|SwitchHasAuthority
    (:Authority
      (Variables|RoundSettings|SetbActiveDuringMinigame (Variables|UserOptions|GetbEnableOnMinigameStart)))))

(event MinigameStateChangedDelegate(FortMinigameLogic)
  (params (Minigame :type Fort%20Minigame%20Object%20Reference) (NewMinigameState :type EFortMinigameState%20Enum :default "PreGame"))
  (switch Utilities|FlowControl|Switch|SwitchonEFortMinigameState NewMinigameState
    (:PreGame
      (bind _currentminigame (Utilities|ServerMinigame)
        (:If%20Valid%20and%20Server
          (if (and (>= (Variables|/Script/FortniteGame.FortMinigame|GetTotalRounds _currentminigame) 1) (== (Variables|/Script/FortniteGame.FortMinigame|GetCurrentRound _currentminigame) 1))
            (Variables|RoundSettings|SetbActiveDuringMinigame (Variables|UserOptions|GetbEnableOnMinigameStart))))))
    (:Warmup
      (CallFunction|HandleinWarmup))
    (:InProgress
      (bind _currentminigame_2 (Utilities|ServerMinigame)
        (:If%20Valid%20and%20Server
          (if (CallFunction|AffectsCurrentRound _currentminigame_2)
            (Utilities|FlowControl|Sequence
              (:then_0
                (Utilities|FlowControl|Delay 0.1) ; Delay this by just a little bit as sometimes during the mini game start all messages are cleared
                (GameplayEvent|BroadcastEvent (Variables|UserOptions|Events|GetOn%20Round%20Start) 0 (Utilities|Struct|MakeGameplayEventRoundStart 0)))
              (:then_1
                (if (Variables|UserOptions|GetbResetClassEachRound)
                  (bind _fortminigamelogic (Variables|Default|GetFortMinigameLogic)
                    (:then
                      (bind _currentminigame_3 (Variables|/Script/FortniteGame.FortMinigameLogicComponent|GetCurrentMinigame _fortminigamelogic)
                        (:then
                          (Utilities|IsValid (Minigame|GetFNEClassComponent _currentminigame_3)
                            (:Is%20Valid
                              (for _array_element (Minigame|GetParticipatingPlayers _currentminigame_3)
                                (Utilities|IsValid _array_element
                                  (:Is%20Valid
                                    (Minigame|SetClassSlotforPlayer (Minigame|GetFNEClassComponent _currentminigame_3) _array_element (Minigame|GameRules|GetDefaultClassSlot (Minigame|GetFNEClassComponent _currentminigame_3) _array_element) (EventDispatchers|CreateEvent :!event OnSetClassSlotForPlayerComplete) (Variables|UserOptions|GetbRespawnPlayerOnClassReset) (Variables|UserOptions|GetbResetCurrentVitalsOnClassReset) (select (Variables|UserOptions|GetbClearAllItemsOnClassReset) "ClearOnClassOrTeamSwitch" "NeverClear") 0.0 3.0)))))))
                        (:else)))
                    (:else)))))))))
    (:PostRoundEnd
      ;[1 1 1] If we have flagged 'Disable Join in Progress at End of Round', then do so here.
      (bind _currentminigame_4 (Utilities|ServerMinigame)
        (:If%20Valid%20and%20Server
          (if (and (CallFunction|AffectsCurrentRound _currentminigame_4) (Variables|UserOptions|GetbDisableJIPAtRoundEnd))
            (bind _matchcomponent (Variables|/Script/FortniteGame.FortMinigame|GetMatchComponent _currentminigame_4)
              (:then
                (Minigame|Matchmaking|SetBackfillandJoininProgressEnabled _matchcomponent false true))
              (:else))))))))

(event Custom|Enable%20Event
  (params (Instigator :type Controller%20Object%20Reference))
  ;[0.6006 0 1] Receivers
  (Variables|RoundSettings|SetbActiveDuringMinigame true))

(event Custom|Disable%20Event
  (params (Instigator :type Controller%20Object%20Reference))
  ;[0.6006 0 1] Receivers
  (Variables|RoundSettings|SetbActiveDuringMinigame false))

(event Custom|Disable%20End%20Round%20Conditions%20Event
  (params (Instigator :type Controller%20Object%20Reference))
  ;[0.6006 0 1] Receivers
  (Utilities|IsActiveandRelevant
    (:True
      (bind _currentminigame (Utilities|ServerMinigame)
        (:If%20Valid%20and%20Server
          (Variables|/Script/FortniteGame.FortMinigameEndConditionComponent|SetbAllowStandardEndGameConditions false (Variables|/Script/FortniteGame.FortMinigame|GetEndConditionComponent _currentminigame)))))))

(event Custom|Disable%20JIP
  (params (Instigator :type Controller%20Object%20Reference))
  ;[0.6006 0 1] Receivers
  (Utilities|IsActiveandRelevant
    (:True
      (CallFunction|CreativeMatchmakingRequest false))))

(event Custom|Enable%20JIP
  (params (Instigator :type Controller%20Object%20Reference))
  ;[0.6006 0 1] Receivers
  (Utilities|IsActiveandRelevant
    (:True
      (CallFunction|CreativeMatchmakingRequest true))))

(event Custom|Toggle%20JIP
  (params (Instigator :type Controller%20Object%20Reference))
  ;[0.6006 0 1] Receivers
  (Utilities|IsActiveandRelevant
    (:True
      (CallFunction|ToggleCreativeMatchmakingRequest))))

(event Custom|End%20Round%20Event
  (params (Instigator :type Controller%20Object%20Reference))
  ;[0.6006 0 1] Receivers
  (Utilities|IsActiveandRelevant
    (:True
      (CallFunction|EndMinigameRound Instigator))))

(event EventPostFinishSpawning
  (UEFNConversionHelper|MigratetoReceiverwhenSameChannel (Variables|Default|GetDisableJoinInProgressWhenReceived) (Variables|Default|GetEnableJoinInProgressWhenReceived) (Variables|Default|GetToggleJoinInProgressWhenReceived))
  (|CallParentFunction))

(fn AffectsCurrentRound
  (params (Minigame :type Fort%20Minigame%20Object%20Reference))
  (Utilities|IsValid Minigame
    (:Is%20Valid
      (return (or (not (Variables|UserOptions|Basic|GetRound_Override)) (== (Variables|/Script/FortniteGame.FortMinigame|GetCurrentRound Minigame) (Variables|UserOptions|Basic|GetRound)))))
    (:Is%20Not%20Valid
      (return false))))

(fn CreativeMatchmakingRequest
  (params (bEnable :type Boolean))
  (bind _currentminigame (Utilities|ServerMinigame)
    (:If%20Valid%20and%20Server
      (bind _matchcomponent (Variables|/Script/FortniteGame.FortMinigame|GetMatchComponent _currentminigame)
        (:then
          (Utilities|IsValidandNotBeingDestroyed self
            (:True
              (Minigame|Matchmaking|SetBackfillandJoininProgressEnabled _matchcomponent bEnable (not bEnable)))))
        (:else)))))

(fn EndMinigameRound
  (params (Requesting%20Player :type Controller%20Object%20Reference))
  ;[1 1 1] End the Round, using the input Player's team as the winning team.
  (bind _currentminigame (Utilities|ServerMinigame)
    (:If%20Valid%20and%20Server
      (LocalVariables|SetMinigame _currentminigame)
      (Utilities|FlowControl|Sequence
        (:then_0
          (bind _asfort_player_controller (Utilities|Casting|CastToFortPlayerController Requesting%20Player))
          (bind _output_get (LocalVariables|SetPC _asfort_player_controller))
          (if (.ReturnValue (Team|GetActorPvPTeamIndex _output_get))
            (Minigame|SetWinningTeamIndex (LocalVariables|GetMinigame) (Team|GetActorPvPTeamIndex _output_get))))
        (:then_1
          (Minigame|EndGame (LocalVariables|GetMinigame) "EndRound"))))))

(fn GetRoundIndex
  (return (select (Variables|UserOptions|Basic|GetRound_Override) (Variables|UserOptions|Basic|GetRound) -1)))

(fn GetRoundNumberVerse
  (return (RoundSettings|GetCurrentRoundNumber)))

(fn HandleInWarmup
  (bind _playercontrolleritr (LocalVariables|GetPlayerControllerItr))
  (bind _resourcetypeitr (LocalVariables|GetResourceTypeItr))
  (Networking|SwitchHasAuthority
    (:Authority
      (Utilities|IsActiveandRelevant
        (:True
          (Utilities|FlowControl|Sequence
            (:then_0
              (RoundSettings|ClearResourcesGivenPerRound)
              (Utilities|FlowControl|Sequence
                (:then_0
                  (RoundSettings|SetResourceGivenPerRound "Wood" (Variables|UserOptions|GetWoodGivenPerRound)))
                (:then_1
                  (RoundSettings|SetResourceGivenPerRound "Stone" (Variables|UserOptions|GetStoneGivenPerRound)))
                (:then_2
                  (RoundSettings|SetResourceGivenPerRound "Metal" (Variables|UserOptions|GetMetalGivenPerRound)))
                (:then_3
                  (RoundSettings|SetResourceGivenPerRound "GoldCurrency" (Variables|UserOptions|GetGoldGivenPerRound)))))
            (:then_1
              (RoundSettings|ClearResourcesGiventoWinnersPerRound)
              (Utilities|FlowControl|Sequence
                (:then_0
                  (RoundSettings|SetResourceGiventoWinnersPerRound "Wood" (Variables|UserOptions|GetWoodGivenToWinnerPerRound)))
                (:then_1
                  (RoundSettings|SetResourceGiventoWinnersPerRound "Stone" (Variables|UserOptions|GetStoneGivenToWinnerPerRound)))
                (:then_2
                  (RoundSettings|SetResourceGiventoWinnersPerRound "Metal" (Variables|UserOptions|GetMetalGivenToWinnerPerRound)))
                (:then_3
                  (RoundSettings|SetResourceGiventoWinnersPerRound "GoldCurrency" (Variables|UserOptions|GetGoldGivenToWinnerPerRound)))))
            (:then_2
              (bind _currentminigame (Variables|/Script/FortniteGame.FortMinigameLogicComponent|GetCurrentMinigame (Variables|Default|GetFortMinigameLogic))
                (:then
                  (Utilities|IsValid (Minigame|Stats|GetMinigameStatsComponent _currentminigame)
                    (:Is%20Valid
                      (LocalVariables|SetRoundWinnerBucket (Class|FortMinigameStatsComponent|GetRoundWinner (Minigame|Stats|GetMinigameStatsComponent _currentminigame) (- (Variables|/Script/FortniteGame.FortMinigame|GetCurrentRound _currentminigame) 1))))))
                (:else)))
            (:then_3
              (bind _currentminigame_2 (Variables|/Script/FortniteGame.FortMinigameLogicComponent|GetCurrentMinigame (Variables|Default|GetFortMinigameLogic))
                (:then
                  (bind _output_get (LocalVariables|SetLocalCurrentMinigame _currentminigame_2))
                  (Utilities|IsValid (Minigame|GetMinigameTeamComponent _output_get)
                    (:Is%20Valid
                      (LocalVariables|SetLocalMinigameTeamComponent (Minigame|GetMinigameTeamComponent _output_get))
                      (for _array_element (Fortnite|GetAllFortPlayerControllers true)
                        (LocalVariables|SetPlayerControllerItr _array_element)
                        (bind _asfort_player_state (Utilities|Casting|CastToFortPlayerState (Variables|/Script/Engine.Controller|GetPlayerState _array_element)))
                        (if (Minigame|IsPlayerinMinigame (LocalVariables|GetLocalCurrentMinigame) _asfort_player_state false)
                          (bind _enumvalue (Utilities|Enum|ForEachEFortResourceType false)
                            (:LoopBody
                              (Utilities|FlowControl|Sequence
                                (:then_0
                                  (LocalVariables|SetResourceTypeItr _enumvalue))
                                (:then_1
                                  (LocalVariables|SetAmountGivenPerRound (RoundSettings|GetResourceGivenPerRound (LocalVariables|GetResourceTypeItr))))
                                (:then_2
                                  (LocalVariables|SetAmountGivenToWinnersPerRound (RoundSettings|GetResourceGiventoWinnersPerRound (LocalVariables|GetResourceTypeItr))))
                                (:then_3
                                  (LocalVariables|SetPlayerBucket (Minigame|GetBucketforPlayerState (LocalVariables|GetLocalMinigameTeamComponent) (Variables|/Script/Engine.Controller|GetPlayerState (LocalVariables|GetPlayerControllerItr)) false)))
                                (:then_4
                                  (bind _returnvalue (Building|K2GetNumAvailableBuildingResources _playercontrolleritr _resourcetypeitr))
                                  (bind _output_get_2 (LocalVariables|SetAmountToGive (Math|Integer|Clamp(Integer) (Math|Integer|Min(Integer) (- (Inventory|GetResourceMaxStackSize _playercontrolleritr _resourcetypeitr) _returnvalue) (+ (LocalVariables|GetAmountGivenPerRound) (select (and (!= (LocalVariables|GetRoundWinnerBucket) -1) (== (LocalVariables|GetPlayerBucket) (LocalVariables|GetRoundWinnerBucket))) (LocalVariables|GetAmountGivenToWinnersPerRound) 0))) 0 (+ (LocalVariables|GetAmountGivenPerRound) (select (and (!= (LocalVariables|GetRoundWinnerBucket) -1) (== (LocalVariables|GetPlayerBucket) (LocalVariables|GetRoundWinnerBucket))) (LocalVariables|GetAmountGivenToWinnersPerRound) 0)))))
                                  (if (> _output_get_2 0)
                                    (Building|K2GiveBuildingResource (LocalVariables|GetPlayerControllerItr) (LocalVariables|GetResourceTypeItr) (LocalVariables|GetAmountToGive))))))))))))
                (:else)))))))))

(fn MigrateLegacyUserOptions
  (bind (_round_override _roundindex) (Class|GetClassDefaults "/CreativeCoreDevices/Device_RoundSettings_V2.Device_RoundSettings_V2_C"))
  (bind (_round_override_2 _round) (Class|GetClassDefaults "/CreativeCoreDevices/Device_RoundSettings_V2.Device_RoundSettings_V2_C"))
  (bind (_round_override_3 _roundindex_2) (Class|GetClassDefaults "/CreativeCoreDevices/Device_RoundSettings_V2.Device_RoundSettings_V2_C"))
  (bind _lastteamstandingwins (Variables|Legacy|GetLastTeamStandingWins))
  (bind _keepitemsbetweenrounds (Variables|Legacy|GetKeepItemsBetweenRounds))
  (bind _reloadandrestockweaponseachround (Variables|Legacy|GetReloadAndRestockWeaponsEachRound))
  (|CallParentFunction)
  (Utilities|FlowControl|Sequence
    (:then_0
      (Utilities|FlowControl|Sequence
        (:then_0
          (if (!= (Variables|Legacy|GetResetClassEachRound) 0)
            (Variables|UserOptions|SetbResetClassEachRound true)
            (Variables|Legacy|SetResetClassEachRound 0)
            (LocalVariables|SetbMigratedProperties true)))
        (:then_1
          (if (!= (Variables|Legacy|GetResetCurrentVitalsOnClassReset) 1)
            (Variables|UserOptions|SetbResetCurrentVitalsOnClassReset false)
            (Variables|Legacy|SetResetCurrentVitalsOnClassReset 1)
            (LocalVariables|SetbMigratedProperties true)))
        (:then_2
          (if (!= (Variables|Legacy|GetRespawnPlayerOnClassReset) 0)
            (Variables|UserOptions|SetbRespawnPlayerOnClassReset true)
            (Variables|Legacy|SetRespawnPlayerOnClassReset 0)
            (LocalVariables|SetbMigratedProperties true)))
        (:then_3
          (if (!= (Variables|Legacy|GetClearAllItemsOnClassReset) 1)
            (Variables|UserOptions|SetbClearAllItemsOnClassReset false)
            (Variables|Legacy|SetClearAllItemsOnClassReset 1)
            (LocalVariables|SetbMigratedProperties true)))
        (:then_4
          (if (!= _keepitemsbetweenrounds -1)
            (Variables|UserOptions|SetbKeepItemsBetweenRoundsOverride true)
            (Variables|UserOptions|SetbKeepItemsBetweenRounds (== _keepitemsbetweenrounds 1))
            (Variables|Legacy|SetKeepItemsBetweenRounds -1)
            (LocalVariables|SetbMigratedProperties true)))
        (:then_5
          (if (!= _reloadandrestockweaponseachround -1)
            (Variables|UserOptions|SetbReloadAndRestockWeaponsEachRoundOverride true)
            (Variables|UserOptions|SetbReloadAndRestockWeaponsEachRound (== _reloadandrestockweaponseachround 1))
            (Variables|Legacy|SetReloadAndRestockWeaponsEachRound -1)
            (LocalVariables|SetbMigratedProperties true)))
        (:then_6
          (if (!= (Variables|Legacy|GetActive) 1)
            (Variables|UserOptions|SetbEnableOnMinigameStart false)
            (Variables|Legacy|SetActive 1)
            (LocalVariables|SetbMigratedProperties true)))
        (:then_7
          (if (!= _lastteamstandingwins -1)
            (Variables|UserOptions|SetbLastStandingWinsOverride true)
            (Variables|UserOptions|SetbLastStandingWins (== _lastteamstandingwins 1))
            (Variables|Legacy|SetLastTeamStandingWins -1)
            (LocalVariables|SetbMigratedProperties true)))
        (:then_8
          (if (!= (Variables|Legacy|GetRoundIndex) _roundindex_2)
            (if (> (Variables|Legacy|GetRoundIndex) 0)
              (Variables|UserOptions|Basic|SetRound_Override true)
              (Variables|UserOptions|Basic|SetRound (Variables|Legacy|GetRoundIndex))
              (else
                (Variables|UserOptions|Basic|SetRound_Override false)
                (Variables|UserOptions|Basic|SetRound _round)))
            (Variables|Legacy|SetRoundIndex _roundindex)
            (LocalVariables|SetbMigratedProperties true)))))
    (:then_1
      (if (LocalVariables|GetbMigratedProperties)
        (OptionData|RefreshPropertyOverrides (Variables|Default|GetToyOptionsComponent))))))

(fn OnSetClassSlotForPlayerComplete
  (params (PlayerState :type Fort%20Player%20State%20Object%20Reference) (ClassSlotIndex :type Byte) (OldClassSlotIndex :type Byte)))

(fn ToggleCreativeMatchmakingRequest
  (bind _currentminigame (Utilities|ServerMinigame)
    (:If%20Valid%20and%20Server
      (bind _matchcomponent (Variables|/Script/FortniteGame.FortMinigame|GetMatchComponent _currentminigame)
        (:then
          (Minigame|Matchmaking|SetBackfillandJoininProgressEnabled _matchcomponent (not (Minigame|Matchmaking|IsGameSessioninBackfillMode _matchcomponent)) (Minigame|Matchmaking|IsGameSessioninBackfillMode _matchcomponent)))
        (:else)))))

(fn ConstructionScript
  (Networking|CanExecuteCosmeticEvents
    (:True
      (Utilities|SetCreativeFakeGlowParameters (Variables|Default|GetGlow) (Variables|Default|GetGlowColor) (Variables|Default|GetFlickerIntensity) (Variables|Default|GetFlickerSpeed) (Variables|Default|GetGlowOpacity)))))
