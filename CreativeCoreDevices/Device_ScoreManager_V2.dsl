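; Fortnite Creative Score Manager device (V2) that awards or sets player score and shows its current value as a floating hologram.
; Team- and enabled-checked events award, increment, decrement, reset or copy player score, updating a clamped value, HUD, audio and trigger-limit/score-match events.
; The cosmetic readout spawns centered digit meshes, coloring them by award type and sign and passing world origin to WPO materials.
; Legacy user options migrate to newer properties.
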(event Custom|Activate%20Event
  (params (Instigator :type Controller%20Object%20Reference))
  ;[0.7957 0.005113 1] Gameplay Functions
  (CallFunction|ActivateScoreDevice Instigator true (select (Variables|Default|GetCurrent%20Increment%20Score%20On%20Awarding) "NewEnumerator1" "NewEnumerator0")))

(event EventBeginPlay
  ;[0.07648 0.25 0.06831] Initialization
  (|CallParentFunction)
  (Effects|RefreshTeamBinding false)
  (Utilities|GetMinigameStatsComponent))

(event Custom|Increment%20Event
  (params (Instigator :type Controller%20Object%20Reference))
  ;[0.7957 0.005113 1] Gameplay Functions
  (CallFunction|ActivateScoreDevice Instigator false "NewEnumerator1") ; Activate with no Instigator
)

(event EventInitializeDevice
  ;[0.07648 0.25 0.06831] Initialization
  (|CallParentFunction)
  (CallFunction|ResetDeviceState true)
  (Effects|SetTeamColor)
  (Effects|RefreshTeamBinding false)
  (CallFunction|HandleBaseVisibility)
  (Effects|UpdateReadout)
  (Networking|SwitchHasAuthority
    (:Authority
      (Class|CreativeUIBroadcastComponent|Initialize (Variables|Default|GetCreative_HUDBroadcast_Component) (Variables|Default|GetFortMinigameProgress) (Variables|UserOptions|Advanced|GetDisplayScoreUpdateOnHUD)))))

(event Custom|Decrement%20Event
  (params (Instigator :type Controller%20Object%20Reference))
  ;[0.7957 0.005113 1] Gameplay Functions
  (CallFunction|ActivateScoreDevice Instigator false "NewEnumerator2") ; Activate with no Instigator
)

(event MinigameStateChangedDelegate(FortMinigameProgress)
  (params (Minigame :type Fort%20Minigame%20Object%20Reference) (NewMinigameState :type EFortMinigameState%20Enum :default "PreGame"))
  ;[0.07648 0.25 0.06831] Initialization
  (Variables|Default|SetEnabled (Class|CreativeCommonDeviceFunctionLibrary2|EnabledPhaseMatch NewMinigameState (Variables|UserOptions|GetEnabledDuringPhase) self)))

(event Custom|Enable%20Event
  (params (Instigator :type Controller%20Object%20Reference))
  ;[0.7957 0.005113 1] Gameplay Functions
  (if (CallFunction|IsTeamValid Instigator)
    (Variables|Default|SetEnabled true)))

(event EventOnAddedToMinigame
  (params (Minigame :type Fort%20Minigame%20Object%20Reference))
  ;[0.07648 0.25 0.06831] Initialization
  (|CallParentFunction Minigame)
  (Minigame|BindEventtoOnPlayModeChangedDelegate Minigame (delegate Custom|On%20Play%20Mode%20Changed))
  (Effects|RefreshTeamBinding true))

(event OnPropertiesLoaded(ToyOptionsComponent)
  ;[0.07648 0.25 0.06831] Initialization
  (CallFunction|MigrateUserOptionsifNeeded))

(event Custom|Disable%20Event
  (params (Instigator :type Controller%20Object%20Reference))
  ;[0.7957 0.005113 1] Gameplay Functions
  (if (CallFunction|IsTeamValid Instigator)
    (Variables|Default|SetEnabled false)))

(event Custom|On%20Play%20Mode%20Changed
  (params (Minigame :type Fort%20Minigame%20Object%20Reference) (bIsInPlayMode :type Boolean))
  ;[0.07648 0.25 0.06831] Initialization
  (CallFunction|HandleBaseVisibility)
  (Effects|UpdateReadout))

(event Custom|ClientPlaySound
  (params (IncreaseScore :type Boolean))
  (bind _audio (Variables|Default|GetAudio))
  ;[0.7801 0.4004 1] Play audio in client
  (Networking|CanExecuteCosmeticEvents
    (:True
      (if (Math|Boolean|ANDBoolean (Default|IsinPlayMode) (Variables|Default|GetEnabled) (Variables|UserOptions|GetPlay%20Audio))
        (Audio|Components|Audio|SetSound _audio (select IncreaseScore (Variables|Default|GetAdd%20Audio) (Variables|Default|GetSubtract%20Audio)))
        (Audio|Components|Audio|Play _audio 0.0)))))

(event EventOnMinigameEnded
  ;[0.07648 0.25 0.06831] Initialization
  (|CallParentFunction)
  (CallFunction|ResetDeviceState true))

(event EventOnRemovedFromMinigame
  (params (Minigame :type Fort%20Minigame%20Object%20Reference))
  ;[0.07648 0.25 0.06831] Initialization
  (|CallParentFunction Minigame)
  (Effects|RefreshTeamBinding false))

(event Custom|Reset%20Event
  (params (Instigator :type Controller%20Object%20Reference))
  (bind _enabled (Variables|Default|GetEnabled))
  ;[0.7957 0.005113 1] Gameplay Functions
  (if (and (CallFunction|IsTeamValid Instigator) _enabled)
    (CallFunction|ResetDeviceState _enabled)))

(event Custom|Activate%20Score%20Device
  (params (Instigator :type Controller%20Object%20Reference) (AwardScore :type Boolean) (Update%20Score :type EAwardScoreType%20Enum :default "NewEnumerator0"))
  ;[0.1719 0.6901 1 0.2] Score Logic
  (if (and (CallFunction|IsTeamValid Instigator) (Variables|Default|GetEnabled))
    (Variables|Default|SetLast%20Instigator (CreativeAI|Controller|GetParticipantControllerfromActor Instigator false))
    (if AwardScore
      (Class|CreativeCommonDeviceFunctionLibrary|AwardScore self (Variables|Default|GetNextScore) (Variables|UserOptions|Basic|GetScore%20Award%20Type) (Variables|Default|GetLast%20Instigator) (Variables|Default|GetFortMinigameProgress))
      (if (!= (Variables|Default|GetNextScore) 0)
        _
        (elif (Variables|UserOptions|Advanced|GetDisplayScoreUpdateIfScoreIs0)
          _
          (else
            (CallFunction|PlayAudioFX)
            (CallFunction|IncrementTimesTriggered)
            (CallFunction|UpdateScoreValue Update%20Score (Variables|Default|GetCurrent%20Score%20Change%20When%20Activated)))))
      (Class|CreativeHUDBroadcastComponent|DoScoreEventforParticipant (Variables|Default|GetCreative_HUDBroadcast_Component) (Variables|Default|GetLast%20Instigator) (Variables|Default|GetNextScore) (Variables|UserOptions|Advanced|GetDisplayScoreUpdateOnHUD))
      (else
        _))))

(event Custom|SetToPlayerScore%20Event
  (params (Instigator :type Controller%20Object%20Reference))
  ;[0.7957 0.005113 1] Gameplay Functions
  (if (CallFunction|IsTeamValid Instigator)
    (CallFunction|SetDeviceScoreToPlayer Instigator)))

(fn ActivateFromOtherDevice
  (params (Instigator :type Controller%20Object%20Reference) (Requester :type ERequesterType%20Enum :default "NewEnumerator0"))
  (CallFunction|ActivateScoreDevice Instigator true (select (Variables|Default|GetCurrent%20Increment%20Score%20On%20Awarding) "NewEnumerator1" "NewEnumerator0"))
  (return true))

(fn AwardScore
  (params (InstigatingPlayer :type Controller%20Object%20Reference) (ScoreToAward :type Integer))
  (if (and (CallFunction|IsTeamValid InstigatingPlayer) (Variables|Default|GetEnabled))
    (Variables|Default|SetLast%20Instigator (CreativeAI|Controller|GetParticipantControllerfromActor InstigatingPlayer))
    (Class|CreativeCommonDeviceFunctionLibrary|AwardScore self ScoreToAward (Variables|UserOptions|Basic|GetScore%20Award%20Type) (Variables|Default|GetLast%20Instigator) (Variables|Default|GetFortMinigameProgress))
    (CallFunction|PlayAudioFX)
    (CallFunction|IncrementTimesTriggered)))

(fn CacheMaterials
  (Networking|CanExecuteCosmeticEvents
    (:True
      (Variables|MaterialRefs|SetSign%20MID (Material|GetMIDfromMaterial (Variables|Default|GetSign) 0 (Variables|MaterialRefs|GetSignMaterial)))
      (Variables|MaterialRefs|GetDigit%20MID
        (:then)
        (:else
          (bind _returnvalue (Rendering|Material|CreateDynamicMaterialInstance (Variables|Effects|GetDigit%20Material) "None" "None" :!class /Script/Engine.KismetMaterialLibrary))
          (Variables|MaterialRefs|SetDigit%20MID _returnvalue))))))

(fn CanInteract
  (params (InteractingPawn :type Fort%20Pawn%20Object%20Reference) (InteractionBeingAttempted :type EInteractionBeingAttempted%20Enum :default "FirstInteraction") (InteractionType :type TInteractionType%20Enum :default "IT_NoInteraction"))
  (return false))

(fn CreateOrReattachScoreColorMIDs
  ;[1 1 1] Reattach cached original materials for MIDs which use score color
  (Material|ReattachOrCreateMID (Variables|Default|GetSign) (Variables|MaterialRefs|GetSign%20MID))
  (Material|ReattachOrCreateMID (Variables|Default|GetLightRays) (Variables|MaterialRefs|GetLightRays%20MID))
  (for _array_element (Variables|Default|GetDigitMeshes)
    (Material|ReattachOrCreateMID _array_element (Variables|MaterialRefs|GetDigit%20MID))))

(fn GetDeviceColor
  (return 0))

(fn GetInstigatingController
  (return 0))

(fn GetCurrentScore
  (params (QueryPlayer :type Controller%20Object%20Reference))
  (bind _controller (CreativeAI|Controller|GetParticipantControllerfromActor QueryPlayer))
  (bind _returnvalue (Minigame|Stats|GetMinigameStatsComponent _controller))
  (bind _playerstate (Variables|/Script/Engine.Controller|GetPlayerState _controller))
  (bind (_asfort_player_state _bsuccess) (Utilities|Casting|CastToFortPlayerState _playerstate))
  (if (CallFunction|IsTeamValid QueryPlayer)
    (Utilities|IsValid _controller
      (:Is%20Valid
        (Utilities|IsValid _returnvalue
          (:Is%20Valid
            (if _bsuccess
              (return (Minigame|Stats|GetStatforPlayer _returnvalue _asfort_player_state (Class|FortMinigameStatsComponent|GetLoadedStatfromClass _returnvalue "/Game/Athena/Playset/Minigames/StatFilters/MinigameStat_Score.MinigameStat_Score_C") false))
              (else
                (return 0))))
          (:Is%20Not%20Valid
            (return 0))))
      (:Is%20Not%20Valid
        (return 0)))
    (else
      (return 0))))

(fn GetIncrementScoreValueOnAward
  (return (Variables|Default|GetCurrent%20Increment%20Score%20On%20Awarding)))

(fn GetScoreIncrementQuantity
  (return (Variables|Default|GetCurrent%20Score%20Change%20When%20Activated)))

(fn GetScoreValue
  (return (Variables|Default|GetNextScore)))

(fn GetSpawnedPickups
  (return 0))

(fn GetVisibilityComponents
  (bind _localviscomponents (LocalVariables|GetLocalVisComponents))
  (Utilities|Array|AppendArray _localviscomponents (Utilities|Array|MakeArray (Variables|Default|GetSign) (Variables|Default|GetPointLight) (Variables|Default|GetLightRays) (Variables|Default|GetAudio)))
  (Utilities|Array|AppendArray _localviscomponents (Variables|Default|GetDigitMeshes))
  (return (LocalVariables|GetLocalVisComponents)))

(fn HandleBaseVisibility
  (LocalVariables|SetBaseIsVisible (or (Utilities|Enum|Equal(Enum) (Variables|UserOptions|GetVisible%20In%20Game) "NewEnumerator1") (not (Class|CreativeCommonDeviceFunctionLibrary|IsLogicinPlayMode (Variables|Default|GetFortMinigameProgress)))))
  (Rendering|SetVisibility (Variables|Debug|GetStaticMeshComponent) (LocalVariables|GetBaseIsVisible) false)
  (Collision|SetCollisionEnabled (Variables|Debug|GetStaticMeshComponent) (select (LocalVariables|GetBaseIsVisible) "QueryAndPhysics" "NoCollision"))
  (Development|SetHiddeninGame (Variables|Debug|GetStaticMeshComponent) (and (Utilities|Enum|NotEqual(Enum) (Variables|UserOptions|GetVisible%20In%20Game) "NewEnumerator1") (Editor|InEditor)) false))

(fn IncrementTimesTriggered
  (bind _output_get (Variables|Default|SetTimesTriggered (+ (Variables|Default|GetTimesTriggered) 1)))
  (if (Math|Boolean|ANDBoolean (== _output_get (Variables|UserOptions|GetTimes%20Can%20Trigger)) (> (Variables|UserOptions|GetTimes%20Can%20Trigger) 0) (Variables|UserOptions|GetTimes%20Can%20Trigger_Override))
    (GameplayEvent|BroadcastEvent (Variables|UserOptions|Events|GetOn%20Max%20Triggers) 0 (Utilities|Struct|MakeGameplayEventonMaxTriggers (Variables|Default|GetLast%20Instigator)))
    (Variables|Default|SetEnabled false)))

(fn IsEnabled
  (return false))

(fn IsHoldingItemForPawn
  (params (Item :type Item%20Variant%20Handle%20Structure) (Pawn :type Pawn%20Object%20Reference))
  (return false))

(fn IsTeamValid
  (params (Actor :type Actor%20Object%20Reference))
  (bind _activating_team (Variables|UserOptions|GetActivating%20Team))
  (if (Creative|TeamOption|IsTeamType _activating_team "Any")
    (return true)
    (else
      (return (== (Creative|TeamOption|ConverttoTeam _activating_team) (Team|GetActorTeam Actor))))))

(fn IsVisibleDuringMinigame
  (return (Utilities|Enum|NotEqual(Enum) (Variables|UserOptions|GetVisible%20In%20Game) "NewEnumerator0")))

(fn MessageScore
  (return (Variables|UserOptions|Advanced|GetHUD%20Message) (Variables|UserOptions|Advanced|GetHUD%20MessageScoreColor) (Variables|UserOptions|Advanced|GetHUD%20MessageColor) (Variables|UserOptions|Advanced|GetResetHUD%20MessageScore)))

(fn MigrateLegacyUserOptions
  (bind _timescantrigger (Class|GetClassDefaults "/CreativeCoreDevices/Device_ScoreManager_V2.Device_ScoreManager_V2_C"))
  (bind _enabledindex (Class|GetClassDefaults "/CreativeCoreDevices/Device_ScoreManager_V2.Device_ScoreManager_V2_C"))
  (|CallParentFunction)
  (Utilities|FlowControl|Sequence
    (:then_0
      ;[1 0.421 0.7323 0.25] Enabled int -> enum conversion (+1, since the int had a negative value).
      (if (!= (Variables|Legacy|GetEnabledIndex) _enabledindex)
        (Variables|UserOptions|SetEnabledDuringPhase (Utilities|Enum|BytetoEnumEEnabledOnPhase (Math|Conversions|ToByte(Integer) (+ (Variables|Legacy|GetEnabledIndex) 1)))) ;     Changing Int to Enum.\nInteger had a -1 option, so we're \nshifting everything +1 to make it\n    match our new enum
        (Variables|Legacy|SetEnabledIndex _enabledindex) ; Set migrated property to \ndefault so won't migrate again.
        (LocalVariables|SetOptions%20Migrated true)))
    (:then_1
      ;[1 0.421 0.7323 0.25] Splitting out Times Can Trigger into int + Times Can Trigger Override.  It will default to infinite.
      (if (!= (Variables|Legacy|GetTimesCanTrigger) _timescantrigger)
        (Variables|UserOptions|SetTimes%20Can%20Trigger (Variables|Legacy|GetTimesCanTrigger))
        (Variables|Legacy|SetTimesCanTrigger _timescantrigger) ; Legacy Variable
        (Variables|UserOptions|SetTimes%20Can%20Trigger_Override true)
        (LocalVariables|SetOptions%20Migrated true)))
    (:then_2
      ;[1 1 1] Refresh property overrides if anything was updated during migration
      (if (LocalVariables|GetOptions%20Migrated)
        (OptionData|RefreshPropertyOverrides (Variables|Default|GetToyOptionsComponent))))))

(fn MigrateUserOptionsIfNeeded
  (bind _toyoptionscomponent (Variables|Default|GetToyOptionsComponent))
  (bind _currentoptions (OptionData|GetPropertyOverrides _toyoptionscomponent))
  (bind _oldvisibilityknob (LocalVariables|GetOldVisibilityKnob))
  (bind (_value _returnvalue) (Utilities|Map|Find _currentoptions _oldvisibilityknob))
  (bind _toyoptionscomponent_2 (Variables|Default|GetToyOptionsComponent))
  (bind _toyoptionscomponent_3 (Variables|Default|GetToyOptionsComponent))
  (if _returnvalue
    (if (Utilities|String|Equal,CaseInsensitive(String) _value "True")
      (OptionData|SetPropertyOverrides _toyoptionscomponent_3 (Utilities|Map|MakeMap (LocalVariables|GetNewVisibilityKnob) "Visible"))
      (OptionData|SetPropertyOverrides _toyoptionscomponent_3 (Utilities|Map|MakeMap (LocalVariables|GetNewPlayAudioKnob) "On"))
      (else
        (OptionData|SetPropertyOverrides _toyoptionscomponent_2 (Utilities|Map|MakeMap (LocalVariables|GetNewVisibilityKnob) "NotVisible"))
        (OptionData|SetPropertyOverrides _toyoptionscomponent_2 (Utilities|Map|MakeMap (LocalVariables|GetNewPlayAudioKnob) "Off"))))
    (OptionData|RemovePropertyOverrides (Variables|Default|GetToyOptionsComponent) (Utilities|Array|MakeArray (LocalVariables|GetOldVisibilityKnob)))))

(fn OnRep_Enabled
  (Networking|CanExecuteCosmeticEvents
    (:True
      (Effects|UpdateOutputVisibility))
    (:False
      (if (not (Variables|Default|GetEnabled))
        (Variables|Default|SetLast%20Instigator 0)))))

(fn OnRep_NextScore
  (Effects|UpdateReadout))

(fn PlayAudioFX
  (CallFunction|ClientPlaySound (>= (Utilities|Select :Index (Variables|UserOptions|Basic|GetScore%20Award%20Type) :NewEnumerator0 0 :NewEnumerator1 (Variables|Default|GetNextScore) :NewEnumerator2 (* (Variables|Default|GetNextScore) -1) :NewEnumerator3 (Variables|Default|GetNextScore)) 0)))

(fn RefreshTeamBinding
  (params (Unbind :type Boolean))
  (Networking|CanExecuteCosmeticEvents
    (:True
      (Teams|RefreshTeamChangeBinding (Variables|Default|GetFortMinigameProgress) Unbind (EventDispatchers|CreateEvent :!event Set%20Team%20Color)))))

(fn ResetDeviceState
  (params (Enable :type Boolean))
  (Networking|SwitchHasAuthority
    (:Authority
      ;[1 1 1] Reset live 'should increment' + 'value' state variables to default to the same value as user option
      (Variables|Default|SetCurrent%20Increment%20Score%20On%20Awarding (Variables|UserOptions|GetIncrement%20Score%20on%20Awarding))
      (Variables|Default|SetCurrent%20Score%20Change%20When%20Activated (Variables|UserOptions|GetScore%20Change%20when%20Activated))
      (Variables|Default|SetTimesTriggered 0)
      (CallFunction|UpdateScoreValue "NewEnumerator3" (Variables|UserOptions|Basic|GetScore%20Value))
      (Variables|Default|SetEnabled Enable))))

(fn SetDeviceScoretoPlayer
  (params (Sending%20Controller :type Controller%20Object%20Reference))
  (bind _playerstate (Variables|/Script/Engine.Controller|GetPlayerState Sending%20Controller)
    (:then
      (bind _asfort_player_state (Utilities|Casting|CastToFortPlayerState _playerstate))
      (bind _output_get (Utilities|GetMinigameStatsComponent)
        (:Valid
          (CallFunction|UpdateScoreValue "NewEnumerator3" (Minigame|Stats|GetStatforPlayer _output_get _asfort_player_state (Class|FortMinigameStatsComponent|GetLoadedStatfromClass _output_get "/Game/Athena/Playset/Minigames/StatFilters/MinigameStat_Score.MinigameStat_Score_C") false)))))
    (:else)))

(fn SetTeamColor
  (Networking|CanExecuteCosmeticEvents
    (:True
      (Teams|UpdateTeamColor (Variables|Default|GetFortMinigameProgress) (Variables|Debug|GetStaticMeshComponent) (select (Creative|TeamOption|IsTeamType (Variables|UserOptions|GetActivating%20Team) "Any") -1 (- (Creative|TeamOption|GetLegacyTeamIndex (Variables|UserOptions|GetActivating%20Team) false) 1))))))

(fn SetTriggerCollision
  (params (On :type Boolean))
  (return false))

(fn SetIncrementScoreValueOnAward
  (params (ShouldIncrement :type Boolean))
  (Networking|SwitchHasAuthority
    (:Authority
      (Variables|Default|SetCurrent%20Increment%20Score%20On%20Awarding ShouldIncrement))))

(fn SetScoreIncrementQuantity
  (params (NewScoreIncrementQuantity :type Integer))
  (Networking|SwitchHasAuthority
    (:Authority
      (Variables|Default|SetCurrent%20Score%20Change%20When%20Activated (Math|Integer|Clamp(Integer) NewScoreIncrementQuantity -200 200)))))

(fn SetScoreValue
  (params (NewScoreValue :type Integer))
  (Networking|SwitchHasAuthority
    (:Authority
      (CallFunction|UpdateScoreValue "NewEnumerator3" NewScoreValue))))

(fn UpdateOutputVisibility
  (Networking|CanExecuteCosmeticEvents
    (:True
      (bind _output_get (LocalVariables|Setlocal%20Visibility (Math|Boolean|ANDBoolean (or (< (Variables|Default|GetTimesTriggered) (Variables|UserOptions|GetTimes%20Can%20Trigger)) (not (Variables|UserOptions|GetTimes%20Can%20Trigger_Override))) (Variables|Default|GetEnabled) (VisibleInGameComponent|GetDesiredVisibility))))
      (Rendering|SetVisibility (Variables|Default|GetLightRays) _output_get false)
      (Rendering|SetVisibility (Variables|Default|GetResultVisualBase) _output_get true)
      (Effects|UpdateReadout))))

(fn UpdateReadout
  (bind _local_value (LocalVariables|Getlocal%20Value))
  (bind _returnvalue (Math|Integer|%25%28Integer%29 _local_value 10))
  (bind _sign (Variables|Default|GetSign))
  (bind _nextscore (Variables|Default|GetNextScore))
  (Networking|CanExecuteCosmeticEvents
    (:True
      (if (and (VisibleInGameComponent|GetDesiredVisibility) (not (Building|IsPreviewBuildingActor)))
        (if (Variables|Default|GetEnabled)
          (LocalVariables|Setlocal%20Sign (* (Math|Integer|Sign(Integer) _nextscore) (select (Utilities|Enum|Equal(Enum) (Variables|UserOptions|Basic|GetScore%20Award%20Type) "NewEnumerator2") -1 1)))
          (bind _output_get (LocalVariables|Setlocal%20Value (Math|Integer|Absolute(Integer) _nextscore)))
          (CallFunction|UpdateDigitCount _output_get)
          (Effects|UpdateScoreDetails)
          (LocalVariables|Setlocal%20Digits (Variables|Default|GetDigitMeshes))
          ;[1 0.6061 0] Set Meshes from the score.
          (bind (_arrayindex _arrayelement) (Utilities|Array|ReverseforEachLoop (LocalVariables|Getlocal%20Digits))
            (:LoopBody
              (Components|StaticMesh|SetStaticMesh _arrayelement (Utilities|Array|Get(acopy) (Variables|Effects|GetDigitsStaticMeshTemplates) _returnvalue))
              (LocalVariables|Setlocal%20Value (/ (- (LocalVariables|Getlocal%20Value) _returnvalue) 10)))
            (:Completed
              ;[1 1 1] Insert the Sign at the start of the array, assuming we're accumulating rather than setting score, and there is score to add/remove.
              (if (and (!= (LocalVariables|Getlocal%20Sign) 0) (or (Utilities|Enum|Equal(Enum) (Variables|UserOptions|Basic|GetScore%20Award%20Type) "NewEnumerator2") (Utilities|Enum|Equal(Enum) (Variables|UserOptions|Basic|GetScore%20Award%20Type) "NewEnumerator1")))
                (Components|StaticMesh|SetStaticMesh (Variables|Default|GetSign) (Utilities|Array|Get(acopy) (Variables|Effects|GetSignStaticMeshes) (Math|Integer|Max(Integer) (LocalVariables|Getlocal%20Sign))))
                (Utilities|Array|Insert (LocalVariables|Getlocal%20Digits) _sign)
                (UEFNConversionHelper|SetComponentHiddenEditor _sign (not (VisibleInGameComponent|GetDesiredVisibility)))
                (else
                  (UEFNConversionHelper|SetComponentHiddenEditor (Variables|Default|GetSign) true)))
              ;[0 0.3855 1] Center Digits
              (LocalVariables|Setlocal%20Start (* (Utilities|Array|LastIndex (LocalVariables|Getlocal%20Digits)) (/ (Variables|Effects|GetDigit%20Kerning) -2.0)))
              (for (_array_element _array_index) (LocalVariables|Getlocal%20Digits)
                (Transformation|SetRelativeLocation _array_element (Math|Vector|MakeVector 0.0 (+ (* _array_index (Variables|Effects|GetDigit%20Kerning)) (LocalVariables|Getlocal%20Start))) false false)))))
        (else
          ;[1 0.1036 0 0.2] We don't want the numbers (and they won't be coming back unless we switch back to edit), so remove them.
          (for _array_element_2 (Variables|Default|GetDigitMeshes)
            (Components|DestroyComponent _array_element_2))
          (Utilities|Array|Clear (Variables|Default|GetDigitMeshes)))))))

(fn UpdateScoreDetails
  (bind _lightrays (Variables|Default|GetLightRays))
  (bind _mid (Material|GetMID _lightrays))
  (bind _resultvisualbase (Variables|Default|GetResultVisualBase))
  (bind _returnvalue (Transformation|GetWorldLocation _resultvisualbase))
  (bind _z (.z _returnvalue))
  (bind _y (.y _returnvalue))
  (bind _x (.x _returnvalue))
  (bind _returnvalue_2 (Math|Color|MakeColor _x _y _z))
  (bind _use_static_hologram (Variables|UserOptions|GetUse%20Static%20Hologram))
  (bind _returnvalue_3 (not _use_static_hologram))
  (bind _returnvalue_4 (Math|Conversions|ToFloat(Boolean) _returnvalue_3))
  (bind _local_color (LocalVariables|Getlocal%20Color))
  (bind _digit_material (Variables|Effects|GetDigit%20Material))
  (bind _signmaterial (Variables|MaterialRefs|GetSignMaterial))
  (Networking|CanExecuteCosmeticEvents
    (:True
      (CallFunction|CreateorReattachScoreColorMIDs)
      (for _index (range (+ (Utilities|Array|LastIndex (Variables|Effects|GetColorPositive)) 1))
        (LocalVariables|Seti _index)
        (LocalVariables|Setlocal%20Color (Utilities|Select :Index (Variables|UserOptions|Basic|GetScore%20Award%20Type) :NewEnumerator0 (select (>= (Variables|Default|GetNextScore) 0) (Utilities|Array|Get(acopy) (Variables|Effects|GetColorPositive) (LocalVariables|Geti)) (Utilities|Array|Get(acopy) (Variables|Effects|GetColorNegative) (LocalVariables|Geti))) :NewEnumerator1 (select (>= (Variables|Default|GetNextScore) 0) (Utilities|Array|Get(acopy) (Variables|Effects|GetColorPositive) (LocalVariables|Geti)) (Utilities|Array|Get(acopy) (Variables|Effects|GetColorNegative) (LocalVariables|Geti))) :NewEnumerator2 (select (>= (Variables|Default|GetNextScore) 0) (Utilities|Array|Get(acopy) (Variables|Effects|GetColorNegative) (LocalVariables|Geti)) (Utilities|Array|Get(acopy) (Variables|Effects|GetColorPositive) (LocalVariables|Geti))) :NewEnumerator3 (select (>= (Variables|Default|GetNextScore) 0) (Utilities|Array|Get(acopy) (Variables|Effects|GetColorPositiveSet) (LocalVariables|Geti)) (Utilities|Array|Get(acopy) (Variables|Effects|GetColorNegativeSet) (LocalVariables|Geti)))))
        (switch int (LocalVariables|Geti)
          (:0
            (for _array_element (Variables|Default|GetDigitMeshes)
              (bind _output_get (LocalVariables|SetCheckElem _array_element))
              (Rendering|Material|SetVectorParameterValue (Material|GetMIDfromMaterial _output_get 0 _digit_material) "ColorInner" _local_color :!class /Script/Engine.MaterialInstanceDynamic)
              (Rendering|Material|SetScalarParameterValue (Material|GetMIDfromMaterial _output_get 0 _digit_material) "RotateWithCamera" _returnvalue_4 :!class /Script/Engine.MaterialInstanceDynamic)
              ;[1 1 1] Because of the way the numbers are controlled by material WPO, when the device is rotated or parented the numbers move in odd ways and sometimes away from the device. This operation passes a position to the number meshes independent of the heirarchy or parent of the device
              (Rendering|Material|SetVectorParameterValue (Material|GetMIDfromMaterial _output_get 0 _digit_material) "OriginPosition" _returnvalue_2 :!class /Script/Engine.MaterialInstanceDynamic))
            (bind _output_get_2 (LocalVariables|SetCheckElem (Variables|Default|GetSign)))
            (Rendering|Material|SetVectorParameterValue (Material|GetMIDfromMaterial _output_get_2 0 _signmaterial) "ColorInner" _local_color :!class /Script/Engine.MaterialInstanceDynamic)
            (Rendering|Material|SetScalarParameterValue (Material|GetMIDfromMaterial _output_get_2 0 _signmaterial) "RotateWithCamera" _returnvalue_4 :!class /Script/Engine.MaterialInstanceDynamic)
            (Rendering|Material|SetVectorParameterValue (Material|GetMIDfromMaterial _output_get_2 0 _signmaterial) "OriginPosition" _returnvalue_2 :!class /Script/Engine.MaterialInstanceDynamic))
          (:1
            (Rendering|Material|SetVectorParameterValue _mid "ColorHigh" (LocalVariables|Getlocal%20Color) :!class /Script/Engine.MaterialInstanceDynamic))
          (:2
            (Rendering|Material|SetVectorParameterValue _mid "ColorLow" (LocalVariables|Getlocal%20Color) :!class /Script/Engine.MaterialInstanceDynamic))
          (:3
            (Rendering|Components|Light|SetLightColor (Variables|Default|GetPointLight) (LocalVariables|Getlocal%20Color) true)))))))

(fn UpdateScoreValue
  (params (UpdateScore :type EAwardScoreType%20Enum :default "NewEnumerator0") (Score :type Integer))
  (Variables|Default|SetNextScore (Math|Conversions|ToInteger(Integer64) (Math|Integer64|Clamp(Integer64) (select (Utilities|Enum|Equal(Enum) UpdateScore "NewEnumerator3") (Math|Conversions|ToInteger64(Integer) Score) (+ (Variables|Default|GetNextScore) (Utilities|Select :Index (Variables|GetUpdateScore) :NewEnumerator0 0 :NewEnumerator1 (Variables|GetScore) :NewEnumerator2 (* (Variables|GetScore) -1) :NewEnumerator3 0))) (Math|Conversions|ToInteger64(Integer) (Variables|UserOptions|GetMinimum%20Score)) (Math|Conversions|ToInteger64(Integer) (Variables|UserOptions|GetMaximum%20Score)))))
  (if (== (Variables|Default|GetNextScore) (Variables|UserOptions|GetSendEventOnScoring))
    (GameplayEvent|BroadcastEvent (Variables|UserOptions|Events|GetOn%20Score%20Output) 0 (Utilities|Struct|MakeGameplayEventonScoreOutput (Variables|Default|GetLast%20Instigator)))))

(fn UpdateDigitCount
  (params (DesiredScore :type Integer))
  (bind _digitmeshes (Variables|Default|GetDigitMeshes))
  (bind _returnvalue (Utilities|Array|LastIndex _digitmeshes))
  (bind _digitmeshes_2 (Variables|Default|GetDigitMeshes))
  (bind _returnvalue_2 (Utilities|Array|Length _digitmeshes_2))
  (bind _output_get (Variables|Default|SetDigits (Utilities|String|Len (Utilities|String|ToString(Integer) DesiredScore))))
  (while (< _output_get _returnvalue_2)
    (if (>= _returnvalue 0)
      (Components|DestroyComponent (Utilities|Array|Get(acopy) _digitmeshes _returnvalue))
      (Utilities|Array|RemoveIndex (Variables|Default|GetDigitMeshes) _returnvalue)
      (LocalVariables|SetChanged true)))
  (while (> _output_get _returnvalue_2)
    (bind _spawnedcomponent (Fortnite|Utilities|SpawnTransientSceneComponent "/Script/Engine.StaticMeshComponent" self (Variables|Default|GetResultVisualBase)))
    (Transformation|SetRelativeLocationAndRotation _spawnedcomponent "0, 0, 0" "0.000000,180.000000,0.000000" false false)
    (Utilities|Array|Add (Variables|Default|GetDigitMeshes) _spawnedcomponent)
    (Rendering|Material|SetMaterial _spawnedcomponent 0 (Variables|MaterialRefs|GetDigit%20MID))
    (LocalVariables|SetChanged true))
  ;[1 0.5059 0 0.25] If number of components changed, update the visibility components (this will include the new list of components)
  (if (LocalVariables|GetChanged)
    (Visibility|SetVisibilityComponents (Variables|Default|GetVisibleInGameComponent) (VisibleinGameComponent|GetVisibilityComponents) true)))

(fn ConstructionScript
  (CallFunction|CacheMaterials)
  (|CallParentFunction))
