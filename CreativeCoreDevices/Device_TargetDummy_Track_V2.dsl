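; Creative device that moves an attached V2 shooting target dummy along a configurable-length track, with optional proximity volume.
; Replicated movement state drives forward, reverse or stop, with position lerped along the track.
; It passes user options and transmitter channels to the target and rebroadcasts hit, knockdown, bullseye, pop and hop events as gameplay events.
; Legacy settings migrate to newer options, and construction updates the editor preview's type, facing and track scale.
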(fn BTargetTrackHandleTransmitters
  (params (Instigator :type Controller%20Object%20Reference) (Event :type E_ShootingTargetTransmitter%20Enum :default "NewEnumerator0")))

(fn Play
  (params (bUnpaused :type Boolean))
  (bind _bfailedtoplay (|CallParentFunction bUnpaused))
  (if (and (not _bfailedtoplay) (not bUnpaused))
    (bind _attachedshootingtarget (Variables|Default|GetAttachedShootingTarget)
      (:then
        (Class|DeviceTargetDummyV2|AddTargetTransmitDispatcher _attachedshootingtarget self))
      (:else
        _))
    (return _bfailedtoplay)))

(fn Stop
  (bind _bfailedtostop (|CallParentFunction))
  (bind _attachedshootingtarget (Variables|Default|GetAttachedShootingTarget)
    (:then
      (Class|DeviceTargetDummyV2|RemoveTargetTransmitDispatcher _attachedshootingtarget self))
    (:else
      _))
  (return _bfailedtostop))

(fn BuildExposedAssetProviderVersion(Implementation)
  (params (Version :type Fort%20Exposed%20Asset%20Provider%20Version%20Structure%20%28by%20ref%29))
  (|CallParentFunction Version)
  (ExposedAssetProvider|AddVersion Version 2)
  (return))

(fn CalculatePlayRate
  ;[1 1 1] Accounts for settings but keeps the movement speed of the target over 512 units uniform
  (return (* (/ 1.0 (Variables|UserOptions|Basic|GetLength)) (Variables|UserOptions|Basic|GetSpeed))))

(fn CollectAssetReferencestoAllow(Implementation)
  (params (Collector :type Fort%20Exposed%20Asset%20Reference%20Collector%20Structure%20%28by%20ref%29))
  (|CallParentFunction Collector)
  (ExposedAssetProvider|CollectObjectPropertyAssetReferences Collector 0 (Variables|Default|GetLeftRightTargetBaseMesh))
  (ExposedAssetProvider|CollectIndirectAssetReference Collector (LocalVariables|GetMI_ShootingTargets))
  (return))

(fn EditorUpdates
  (if (Editor|InEditor)
    (bind _asdevice_target_dummy_v2 (Utilities|Casting|CastToDevice_TargetDummy_V2 (Variables|/Script/Engine.ChildActorComponent|GetChildActor (Variables|Default|GetEditorTarget))))
    (Variables|/CreativeCoreDevices/Device_TargetDummy_V2.Device_TargetDummy_V2_C|SetTargetType (Variables|UserOptions|Basic|GetTargetType) _asdevice_target_dummy_v2)
    (CallFunction|SetTargetRotationandBaseMesh)
    (CallFunction|SetTrackLength)))

(fn GetCollisionStaticMesh
  (return "0, 0, 0" (Variables|/Script/Engine.StaticMeshComponent|GetStaticMesh (Variables|Default|GetSM_Target_Track))))

(fn GetFinalDestinationOfTarget
  (return (Math|Vector|MakeVector (* (Variables|UserOptions|Basic|GetLength) 512.0))))

(fn GetMeshComponents
  (return 0))

(fn HandleTargetHitEventBroadcasts
  (params (Instigator :type Controller%20Object%20Reference) (Event :type E_ShootingTargetTransmitter%20Enum :default "NewEnumerator0"))
  (if (Class|CreativeCommonDeviceFunctionLibrary|IsinPlayMode self)
    (switch Utilities|FlowControl|Switch|SwitchonE_ShootingTargetTransmitter Event
      (:NewEnumerator1
        (bind _returnvalue (GameplayEvent|BroadcastEvent (Variables|UserOptions|Events|GetOn%20Hit) 0 (Utilities|Struct|MakeGameplayEventonTargetTrackTargetHit (Variables|Actor|GetInstigator))))
        (return _returnvalue))
      (:NewEnumerator2
        (bind _returnvalue_2 (GameplayEvent|BroadcastEvent (Variables|UserOptions|Events|GetOn%20Knockdown) 0 (Utilities|Struct|MakeGameplayEventonTargetTrackKnockdown (Variables|Actor|GetInstigator))))
        (return _returnvalue_2))
      (:NewEnumerator3
        (bind _returnvalue_3 (GameplayEvent|BroadcastEvent (Variables|UserOptions|Events|GetOn%20Bullseye%20Hit) 0 (Utilities|Struct|MakeGameplayEventonTargetTrackBullseye (Variables|Actor|GetInstigator))))
        (return _returnvalue_3))
      (:NewEnumerator4
        (bind _returnvalue_4 (GameplayEvent|BroadcastEvent (Variables|UserOptions|Events|GetOn%20Pop-Up) 0 (Utilities|Struct|MakeGameplayEventonTargetTrackPopup (Variables|Actor|GetInstigator))))
        (return _returnvalue_4))
      (:NewEnumerator5
        (bind _returnvalue_5 (GameplayEvent|BroadcastEvent (Variables|UserOptions|Events|GetOn%20Pop-Down) 0 (Utilities|Struct|MakeGameplayEventonTargetTrackPopdown (Variables|Actor|GetInstigator))))
        (return _returnvalue_5))
      (:NewEnumerator6
        (bind _returnvalue_6 (GameplayEvent|BroadcastEvent (Variables|UserOptions|Events|GetOn%20Hop-Up) 0 (Utilities|Struct|MakeGameplayEventonTargetTrackHopup (Variables|Actor|GetInstigator))))
        (return _returnvalue_6))
      (:NewEnumerator7
        (bind _returnvalue_7 (GameplayEvent|BroadcastEvent (Variables|UserOptions|Events|GetOn%20Hop-Down) 0 (Utilities|Struct|MakeGameplayEventonTargetTrackHopdown (Variables|Actor|GetInstigator))))
        (return _returnvalue_7)))))

(fn IgnoreWorldGridScaleOverride
  (return false))

(fn InitializeTargetChannels
  (bind _attachedshootingtarget (Variables|Default|GetAttachedShootingTarget))
  (Networking|SwitchHasAuthority
    (:Authority
      (Class|FortGameplayMessageComponentBase|SetChannelId (Variables|/CreativeCoreDevices/Device_TargetDummy_V2.Device_TargetDummy_V2_C|GetHitTransmitter _attachedshootingtarget) (Class|FortGameplayMessageComponentBase|GetChannelId (Variables|Default|GetHitTransmitter)))
      (Class|FortGameplayMessageComponentBase|SetChannelId (Variables|/CreativeCoreDevices/Device_TargetDummy_V2.Device_TargetDummy_V2_C|GetOnHitKnockdownTransmitOn _attachedshootingtarget) (Class|FortGameplayMessageComponentBase|GetChannelId (Variables|Default|GetOnHitKnockdownTransmitOn)))
      (Class|FortGameplayMessageComponentBase|SetChannelId (Variables|/CreativeCoreDevices/Device_TargetDummy_V2.Device_TargetDummy_V2_C|GetOnBullseyeTransmitOn _attachedshootingtarget) (Class|FortGameplayMessageComponentBase|GetChannelId (Variables|Default|GetOnBullseyeTransmitOn)))
      (Class|FortGameplayMessageComponentBase|SetChannelId (Variables|/CreativeCoreDevices/Device_TargetDummy_V2.Device_TargetDummy_V2_C|GetOnPopUpTransmitOn _attachedshootingtarget) (Class|FortGameplayMessageComponentBase|GetChannelId (Variables|Default|GetOnPopUpTransmitOn)))
      (Class|FortGameplayMessageComponentBase|SetChannelId (Variables|/CreativeCoreDevices/Device_TargetDummy_V2.Device_TargetDummy_V2_C|GetOnPopDownTransmitOn _attachedshootingtarget) (Class|FortGameplayMessageComponentBase|GetChannelId (Variables|Default|GetOnPopDownTransmitOn)))
      (Class|FortGameplayMessageComponentBase|SetChannelId (Variables|/CreativeCoreDevices/Device_TargetDummy_V2.Device_TargetDummy_V2_C|GetOnHopUpTransmitOn _attachedshootingtarget) (Class|FortGameplayMessageComponentBase|GetChannelId (Variables|Default|GetOnHopUpTransmitOn)))
      (Class|FortGameplayMessageComponentBase|SetChannelId (Variables|/CreativeCoreDevices/Device_TargetDummy_V2.Device_TargetDummy_V2_C|GetOnHopDownTransmitOn _attachedshootingtarget) (Class|FortGameplayMessageComponentBase|GetChannelId (Variables|Default|GetOnHopDownTransmitOn))))))

(fn MergeTrackAndTargetOptions
  (bind _attachedshootingtarget (Variables|Default|GetAttachedShootingTarget))
  ;[1 1 1] Pass our property overrides to our attached target
  (if (and (Utilities|IsValid _attachedshootingtarget :!class /Script/Engine.KismetSystemLibrary) (Networking|HasAuthority))
    (OptionData|SetPropertyOverrides (Variables|/CreativeCoreDevices/Device_TargetDummy_V2.Device_TargetDummy_V2_C|GetToyOptionsComponent _attachedshootingtarget) (OptionData|GetPropertyOverrides (Variables|Default|GetToyOptionsComponent)))))

(fn MigrateLegacyUserOptions
  (bind (_damagetype _usersuppresshealthbar) (Class|GetClassDefaults "/CreativeCoreDevices/Device_TargetDummy_Track_V2.Device_TargetDummy_Track_V2_C"))
  (bind (_damagetype_2 _proximitytriggersizesetting) (Class|GetClassDefaults "/CreativeCoreDevices/Device_TargetDummy_Track_V2.Device_TargetDummy_Track_V2_C"))
  (bind (_damagetype_3 _popupdelaysetting) (Class|GetClassDefaults "/CreativeCoreDevices/Device_TargetDummy_Track_V2.Device_TargetDummy_Track_V2_C"))
  (bind (_damagetype_4 _rotationsetting) (Class|GetClassDefaults "/CreativeCoreDevices/Device_TargetDummy_Track_V2.Device_TargetDummy_Track_V2_C"))
  (bind (_damagetype_5 _jumptimesetting) (Class|GetClassDefaults "/CreativeCoreDevices/Device_TargetDummy_Track_V2.Device_TargetDummy_Track_V2_C"))
  (bind (_damagetype_6 _hoplengthsetting) (Class|GetClassDefaults "/CreativeCoreDevices/Device_TargetDummy_Track_V2.Device_TargetDummy_Track_V2_C"))
  (bind (_damagetype_7 _startinghealthsetting) (Class|GetClassDefaults "/CreativeCoreDevices/Device_TargetDummy_Track_V2.Device_TargetDummy_Track_V2_C"))
  (bind (_damagetype_8 _maxresetssetting) (Class|GetClassDefaults "/CreativeCoreDevices/Device_TargetDummy_Track_V2.Device_TargetDummy_Track_V2_C"))
  (bind (_damagetype_9 _maxpopupssetting) (Class|GetClassDefaults "/CreativeCoreDevices/Device_TargetDummy_Track_V2.Device_TargetDummy_Track_V2_C"))
  (bind (_damagetype_10 _resetdelaysetting) (Class|GetClassDefaults "/CreativeCoreDevices/Device_TargetDummy_Track_V2.Device_TargetDummy_Track_V2_C"))
  (bind (_damagetype_11 _timebeforehidingsetting) (Class|GetClassDefaults "/CreativeCoreDevices/Device_TargetDummy_Track_V2.Device_TargetDummy_Track_V2_C"))
  (bind _popupdelaysetting_2 (Variables|Legacy|GetPopupDelaySetting))
  (bind _usersuppresshealthbar_2 (Variables|Legacy|GetUserSuppressHealthBar))
  (bind _proximitytriggersizesetting_2 (Variables|Legacy|GetProximityTriggerSizeSetting))
  (bind _rotationsetting_2 (Variables|Legacy|GetRotationSetting))
  (bind _jumptimesetting_2 (Variables|Legacy|GetJumpTimeSetting))
  (bind _hoplengthsetting_2 (Variables|Legacy|GetHopLengthSetting))
  (bind _startinghealthsetting_2 (Variables|Legacy|GetStartingHealthSetting))
  (bind _maxresetssetting_2 (Variables|Legacy|GetMaxResetsSetting))
  (bind _maxpopupssetting_2 (Variables|Legacy|GetMaxPopUpsSetting))
  (bind _resetdelaysetting_2 (Variables|Legacy|GetResetDelaySetting))
  (bind _timebeforehidingsetting_2 (Variables|Legacy|GetTimeBeforeHidingSetting))
  (|CallParentFunction)
  (Utilities|FlowControl|Sequence
    (:then_0
      (Utilities|FlowControl|Sequence
        (:then_0
          ;[0.598 1 0.6062] Migrate Movement Start Range / Set Move On Proximity
          (if (!= _proximitytriggersizesetting _proximitytriggersizesetting_2)
            (bind _output_get (Variables|UserOptions|SetMove%20on%20Proximity (>= _proximitytriggersizesetting_2 0)))
            (if _output_get
              (Variables|UserOptions|SetMovement%20Start%20Range _proximitytriggersizesetting_2)
              (else
                (Variables|UserOptions|SetMovement%20Start%20Range 1.0)))
            (Variables|Legacy|SetProximityTriggerSizeSetting _proximitytriggersizesetting)
            (LocalVariables|SetMigratedOptions true)))
        (:then_1
          ;[0.2917 0.6054 1] Migrate Target Facing Direction
          (if (!= _rotationsetting _rotationsetting_2)
            (Variables|UserOptions|SetTarget%20Facing%20Direction (Utilities|Select :Index _rotationsetting_2 :Option%200 "NewEnumerator3" :Option%201 "NewEnumerator2" :Option%202 "NewEnumerator4" :Option%203 "NewEnumerator1"))
            (Variables|Legacy|SetRotationSetting _rotationsetting_2)
            (LocalVariables|SetMigratedOptions true)))
        (:then_2
          ;[0.3854 0 0.03029] Migrate HoppingFrequencyTime / Set HoppingFrequencyType
          (if (!= _jumptimesetting _jumptimesetting_2)
            (if (> _jumptimesetting_2 0)
              (Variables|UserOptions|SetHoppingFrequencyType "NewEnumerator0")
              (elif (== _jumptimesetting_2 0)
                (Variables|UserOptions|SetHoppingFrequencyType "NewEnumerator1")
                (else
                  (Variables|UserOptions|SetHoppingFrequencyType "NewEnumerator2")
                  (Variables|UserOptions|Basic|SetHoppingFrequencyTime 1.0)
                  (Variables|Legacy|SetJumpTimeSetting _jumptimesetting)
                  (LocalVariables|SetMigratedOptions true))))
            (Variables|UserOptions|Basic|SetHoppingFrequencyTime _jumptimesetting_2)))
        (:then_3
          ;[0 0.08885 1] Migrate HopLengthTime / Set HopLengthType
          (if (!= _hoplengthsetting _hoplengthsetting_2)
            (if (> _hoplengthsetting_2 0)
              (Variables|UserOptions|SetHopLengthType "NewEnumerator0")
              (elif (== _hoplengthsetting_2 0)
                (Variables|UserOptions|SetHopLengthType "NewEnumerator1")
                (else
                  (Variables|UserOptions|SetHopLengthType "NewEnumerator2")
                  (Variables|UserOptions|SetHopLengthTime 2.0)
                  (Variables|Legacy|SetHopLengthSetting _hoplengthsetting)
                  (LocalVariables|SetMigratedOptions true))))
            (Variables|UserOptions|SetHopLengthTime _hoplengthsetting_2)))
        (:then_4
          ;[0.964 1 0] Migrate StartingHealth / Set InfiniteHealth
          (if (!= _startinghealthsetting _startinghealthsetting_2)
            (bind _output_get_2 (Variables|UserOptions|SetInfiniteHealth (< _startinghealthsetting_2 0)))
            (if _output_get_2
              (Variables|UserOptions|Basic|SetStartingHealth 1.0)
              (else
                (Variables|UserOptions|Basic|SetStartingHealth _startinghealthsetting_2)))
            (Variables|Legacy|SetStartingHealthSetting _startinghealthsetting)
            (LocalVariables|SetMigratedOptions true)))
        (:then_5
          ;[0.1949 0.01829 1] Migrate MaximumResets / Set InfiniteResets
          (if (!= _maxresetssetting _maxresetssetting_2)
            (bind _output_get_3 (Variables|UserOptions|SetInfiniteResets (== _maxresetssetting_2 99)))
            (if _output_get_3
              (Variables|UserOptions|SetMaximumResets 0)
              (else
                (Variables|UserOptions|SetMaximumResets _maxresetssetting_2)))
            (Variables|Legacy|SetMaxResetsSetting _maxresetssetting)
            (LocalVariables|SetMigratedOptions true)))
        (:then_6
          ;[0.07896 0.5521 0.06326] Migrate MaximumPop-ups / Set InfinitePop-ups
          (if (!= _maxpopupssetting _maxpopupssetting_2)
            (bind _output_get_4 (Variables|UserOptions|SetInfinitePop-ups (== _maxpopupssetting_2 99)))
            (if _output_get_4
              (Variables|UserOptions|SetMaximumPop-ups 0)
              (else
                (Variables|UserOptions|SetMaximumPop-ups _maxpopupssetting_2)))
            (Variables|Legacy|SetMaxPopUpsSetting _maxpopupssetting)
            (LocalVariables|SetMigratedOptions true)))
        (:then_7
          ;[1 0.3602 0] Migrate ResetTime / Set Reset Time Type
          (if (!= _resetdelaysetting _resetdelaysetting_2)
            (if (> _resetdelaysetting_2 0)
              (Variables|UserOptions|SetReset%20Time%20Type "NewEnumerator0")
              (elif (== _resetdelaysetting_2 0)
                (Variables|UserOptions|SetReset%20Time%20Type "NewEnumerator1")
                (else
                  (Variables|UserOptions|SetReset%20Time%20Type "NewEnumerator2")
                  (Variables|UserOptions|SetResetTime 2.0)
                  (Variables|Legacy|SetResetDelaySetting _resetdelaysetting)
                  (LocalVariables|SetMigratedOptions true))))
            (Variables|UserOptions|SetResetTime _resetdelaysetting_2)))
        (:then_8
          ;[1 0 0.6724] Migrate TimeBeforeHiding / Set TimeBeforeHidingType
          (if (!= _timebeforehidingsetting _timebeforehidingsetting_2)
            (if (> _timebeforehidingsetting_2 0)
              (Variables|UserOptions|SetTimeBeforeHidingType "NewEnumerator0")
              (elif (== _timebeforehidingsetting_2 0)
                (Variables|UserOptions|SetTimeBeforeHidingType "NewEnumerator1")
                (else
                  (Variables|UserOptions|SetTimeBeforeHidingType "NewEnumerator2")
                  (Variables|UserOptions|Basic|SetTimeBeforeHiding 1.0)
                  (Variables|Legacy|SetTimeBeforeHidingSetting _timebeforehidingsetting)
                  (LocalVariables|SetMigratedOptions true))))
            (Variables|UserOptions|Basic|SetTimeBeforeHiding _timebeforehidingsetting_2)))
        (:then_9
          ;[0.2292 0.04614 0.0141] Migrate TimeBeforePop-Up / Set Pop Up Delay Type
          (if (!= _popupdelaysetting _popupdelaysetting_2)
            (if (> _popupdelaysetting_2 0)
              (Variables|UserOptions|SetPop%20Up%20Delay%20Type "NewEnumerator0")
              (elif (== _popupdelaysetting_2 0)
                (Variables|UserOptions|SetPop%20Up%20Delay%20Type "NewEnumerator1")
                (else
                  (Variables|UserOptions|SetPop%20Up%20Delay%20Type "NewEnumerator2")
                  (Variables|UserOptions|SetTimeBeforePop-Up 2.0)
                  (Variables|Legacy|SetPopupDelaySetting _popupdelaysetting)
                  (LocalVariables|SetMigratedOptions true))))
            (Variables|UserOptions|SetTimeBeforePop-Up _popupdelaysetting_2)))
        (:then_10
          ;[1 1 1] Migrate ShowHealthBar
          (if (!= _usersuppresshealthbar _usersuppresshealthbar_2)
            (Variables|UserOptions|Basic|SetShowHealthBar (not _usersuppresshealthbar_2))
            (Variables|Legacy|SetUserSuppressHealthBar _usersuppresshealthbar)
            (LocalVariables|SetMigratedOptions true)))))
    (:then_1
      (if (LocalVariables|GetMigratedOptions)
        (OptionData|RefreshPropertyOverrides (Variables|Default|GetToyOptionsComponent))))))

(fn OnRep_AttachedShootingTarget
  (CallFunction|OnAttachedShootingTargetLoaded))

(fn OnRep_CurrentDirection
  (CallFunction|UpdateforRepNotify))

(fn OnRep_CurrentLerpValue)

(fn OnRep_CurrentMovementState
  (switch int (Math|Conversions|ToInteger(Byte) (Variables|Default|GetCurrentMovementState))
    (:Default
      (CallFunction|StopTargetMovement))
    (:0
      (CallFunction|StopTargetMovement))
    (:1
      (CallFunction|MoveTarget) ; Left to Right movement
    )
    (:2
      (CallFunction|ReverseTargetMovement) ; Right to Left movement
    )))

(fn OnRep_IsCurrentlyMoving
  (CallFunction|UpdateforRepNotify))

(fn OnRep_IsTargetUp
  (CallFunction|UpdateforRepNotify))

(fn OnRep_ServerMoveStartTime
  (CallFunction|UpdateforRepNotify))

(fn SetAudioSpeedSound
  (params (Target%20Speed :type Float%20%28double-precision%29))
  (Utilities|FlowControl|Sequence
    (:then_0
      (if (<= Target%20Speed 0.5)
        (Variables|Audio|SetAudio%20Speed%20Parameter 0)))
    (:then_1
      (if (and (> Target%20Speed 0.5) (< Target%20Speed 1.0))
        (Variables|Audio|SetAudio%20Speed%20Parameter 1)))
    (:then_2
      (if (>= Target%20Speed 1.0)
        (Variables|Audio|SetAudio%20Speed%20Parameter 2))))
  (Audio|Parameter|SetIntegerParameter (Variables|Default|GetTarget_Move_Loop_Cue) "TargetSpeed" (Variables|Audio|GetAudio%20Speed%20Parameter) :!class /Script/AudioExtensions.AudioParameterControllerInterface))

(fn SetMovementState
  (params (State :type Byte))
  (Networking|SwitchHasAuthority
    (:Authority
      (if (Math|Byte|NotEqual(Byte) State (Variables|Default|GetCurrentMovementState))
        (Variables|Default|SetCurrentMovementState State)))))

(fn SetTargetRotationAndBaseMesh
  (bind _targetattachpoint (Variables|Default|GetTargetAttachPoint))
  (switch Utilities|FlowControl|Switch|SwitchonEDeviceRelativeDirection (Variables|UserOptions|GetTarget%20Facing%20Direction)
    (:NewEnumerator1
      (Transformation|SetRelativeRotation _targetattachpoint "0, 0, 0" false false) ; Forward
    )
    (:NewEnumerator2
      (Transformation|SetRelativeRotation _targetattachpoint "0,-90.000000, 0" false false) ; Left
    )
    (:NewEnumerator3
      (Transformation|SetRelativeRotation _targetattachpoint "0,90.000000, 0" false false) ; Right
    )
    (:NewEnumerator4
      (Transformation|SetRelativeRotation _targetattachpoint "0,180.000000, 0" false false) ; Backward
    ))
  (Utilities|IsValid (Variables|Default|GetAttachedShootingTarget)
    (:Is%20Valid
      (CallFunction|AsyncLoadTargetBase))))

(fn SetTrackLength
  (Transformation|SetRelativeScale3D (Variables|Default|GetSM_Target_Track) (Math|Vector|MakeVector (Variables|UserOptions|Basic|GetLength) 1.0 1.0)))

(fn SetupProximityVolume
  (bind _box (Variables|Default|GetBox))
  (bind (_relativelocation_x _relativelocation_y) (Math|Vector|BreakVector (Variables|/Script/Engine.SceneComponent|GetRelativeLocation_X _box)))
  (bind (_boxextent_x _boxextent_y _boxextent_z) (Math|Vector|BreakVector (Variables|/Script/Engine.BoxComponent|GetBoxExtent_X _box)))
  (if (Variables|UserOptions|GetMove%20on%20Proximity)
    (Components|Box|SetBoxExtent _box (Math|Vector|MakeVector (+ (* (Variables|UserOptions|GetMovement%20Start%20Range) 512.0) (* (Variables|UserOptions|Basic|GetLength) 512.0)) (* (Variables|UserOptions|GetMovement%20Start%20Range) 512.0) (/ (* (Variables|UserOptions|GetMovement%20Start%20Range) 512.0) 2.0)) true)
    (Transformation|SetRelativeLocation _box (Math|Vector|MakeVector _relativelocation_x _relativelocation_y _boxextent_z) false false)))

(fn UpdateForRepNotify)

(fn UpdateLerpedPosition
  (params (Alpha :type Float%20%28double-precision%29))
  (bind _destination (CallFunction|GetFinalDestinationOfTarget))
  (Transformation|SetRelativeLocation (Variables|Default|GetTargetAttachPoint) (Math|Vector|Lerp(Vector) "0, 0, 0" _destination Alpha) false false))

(fn UpdateTargetSettings
  (bind _attachedshootingtarget (Variables|Default|GetAttachedShootingTarget)
    (:then
      (UserOptionValues|SetUserOptionValues _attachedshootingtarget 0 (UserOptionValues|GetUserOptionValues)))
    (:else)))

(fn ConstructionScript
  (|CallParentFunction)
  (CallFunction|EditorUpdates))
