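; Creative Team Settings & Inventory device that applies per-team player rules, grants starting items and broadcasts team spawn, elimination and out-of-respawns events.
; Options register with the minigame's team and weapon mutators and become scalable-float values, resolved through a data-registry scope built from team and spatial interface.
; Drop-box hits queue pickups into the item list, granted on teleport, possession or team change.
; Version handlers migrate legacy options.
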(event PostMinigameStartedDelegate(FortMinigameProgress)
  ;[0.005519 0.3952 0.9479 0.307] Player Spawned (or Minigame Started, so first place the Event Binding system wants to fire spawned events during a game).
  (if (Variables|UserOptions|Mode|Spawning|GetbSpawnEventActivatesForPlayers)
    (bind _currentminigame (Variables|/Script/FortniteGame.FortMinigameLogicComponent|GetCurrentMinigame (Variables|Default|GetFortMinigameProgress))
      (:then
        (bind _outplayers (Minigame|GetParticipatingPlayers_NotPure _currentminigame))
        (for _array_element _outplayers
          (Utilities|IsValid _array_element
            (:Is%20Valid
              (CallFunction|OnPlayerSpawned (Fortnite|GetFortPlayerControllerfromActor _array_element))))))
      (:else))))

(event Custom|On%20Agent%20Spawned
  (params (Controller :type Controller%20Object%20Reference))
  ;[0.005519 0.3952 0.9479 0.307] Player Spawned (or Minigame Started, so first place the Event Binding system wants to fire spawned events during a game).
  (Utilities|IsValid Controller
    (:Is%20Valid
      (bind _currentminigame (Variables|/Script/FortniteGame.FortMinigameLogicComponent|GetCurrentMinigame (Variables|Default|GetFortMinigameProgress))
        (:then
          (Minigame|IsInProgress _currentminigame
            (:Yes
              (if (CallFunction|IsActoronSameTeam Controller)
                (if (and (Utilities|Enum|NotEqual(Enum) (Variables|/Script/FortniteGame.FortMinigame|GetCurrentState _currentminigame) "Transitioning") (Utilities|Enum|NotEqual(Enum) (Variables|/Script/FortniteGame.FortMinigame|GetCurrentState _currentminigame) "WaitingForCameras"))
                  (GameplayEvent|BroadcastEvent (Variables|UserOptions|Events|GetOnTeamMemberSpawned) 0 (Utilities|Struct|MakeGameplayEventSpawned Controller)))))))
        (:else)))))

(event Custom|On%20AI%20Spawned
  (params (Pawn :type Pawn%20Object%20Reference) (RequestID :type Integer))
  ;[0.005519 0.3952 0.9479 0.307] Player Spawned (or Minigame Started, so first place the Event Binding system wants to fire spawned events during a game).
  (Utilities|IsValid Pawn
    (:Is%20Valid
      (CallFunction|OnAgentSpawned (Pawn|GetController Pawn)))))

(event Custom|On%20Player%20Spawned
  (params (SpawningPlayer :type Controller%20Object%20Reference))
  ;[0.005519 0.3952 0.9479 0.307] Player Spawned (or Minigame Started, so first place the Event Binding system wants to fire spawned events during a game).
  (CallFunction|OnAgentSpawned SpawningPlayer))

(event EventInitializeDevice
  ;[0.08761 1 0 0.5] Initialization
  (|CallParentFunction)
  (TeamSettings|AddTeamSettingstoMinigame (TeamSettings|GetCurrentMinigame))
  (CallFunction|UpdateTeamAssignmentDisplay)
  (DeviceVisuals|UpdateColors))

(event EditorPreview|EventSetupForEditorPreview
  ;[0.08761 1 0 0.5] Initialization
  (CallFunction|SetupItemPreview)
  (CallFunction|UpdateTeamAssignmentDisplay)
  (DeviceVisuals|UpdateColors))

(event Custom|Setup%20Respawn%20Item%20Grant%20Binds
  (params (Minigame :type Fort%20Minigame%20Object%20Reference))
  ;[0.4374 0.9901 1] When player pawn is possessed give the player the items
  (Networking|SwitchHasAuthority
    (:Authority
      (Utilities|IsValid Minigame
        (:Is%20Valid
          (for _array_element (Minigame|GetParticipatingPlayers Minigame)
            (Utilities|IsValid _array_element
              (:Is%20Valid
                (Utilities|IsValid (Fortnite|GetControllerfromActor _array_element)
                  (:Is%20Valid
                    (Pawn|BindEventtoOnPossessedPawnChanged (Fortnite|GetControllerfromActor _array_element) (delegate Custom|OnPossessedPawnChanged)))))))
          (Minigame|BindEventtoPlayerAddedDelegate Minigame (delegate Custom|OnPlayerAddedToMinigame)))))))

(event EventBeginPlay
  ;[0.08761 1 0 0.5] Initialization
  (|CallParentFunction)
  (Networking|SwitchHasAuthority
    (:Authority
      (CallFunction|MigrateDatatoPickupList)
      (EventDispatchers|BindEventtoOnItemAdded (Variables|Default|GetPickupItemList) (EventDispatchers|CreateEvent :!event Custom|OnAddedToInventoryList)))
    (:Remote
      (CallFunction|SetupItemPreview)))
  (Default|AssignOnItemListUpdated (Variables|Default|GetPickupItemList) (EventDispatchers|CreateEvent :!event Custom|OnItemListUpdated)))

(event Custom|OnPlayerAddedToMinigame
  (params (UniqueNetId :type Unique%20Net%20Id%20Structure) (bIsLocalPlayer :type Boolean))
  (bind _returnvalue (Fortnite|Athena|GameState|GetGameStateAthena))
  (bind _returnvalue_2 (Class|FortGameState|GetFortPlayerStatefromUniqueNetId _returnvalue UniqueNetId))
  (bind _returnvalue_3 (Fortnite|GetControllerfromActor _returnvalue_2))
  ;[0.4374 0.9901 1] When player pawn is possessed give the player the items
  (Utilities|IsValid _returnvalue
    (:Is%20Valid
      (Utilities|IsValid _returnvalue_2
        (:Is%20Valid
          (Utilities|IsValid _returnvalue_3
            (:Is%20Valid
              (Pawn|BindEventtoOnPossessedPawnChanged _returnvalue_3 (delegate Custom|OnPossessedPawnChanged)))))))))

(event OnMeshandMaterialsLoaded(CreativeEditOnlyMesh)
  ;[0.08761 1 0 0.5] Initialization
  (DeviceVisuals|UpdateColors)
  (CallFunction|UpdateTeamAssignmentDisplay))

(event Custom|OnPossessedPawnChanged
  (params (OldPawn :type Pawn%20Object%20Reference) (NewPawn :type Pawn%20Object%20Reference))
  ;[0.4374 0.9901 1] When player pawn is possessed give the player the items
  (bind _asfort_player_pawn (Utilities|Casting|CastToFortPlayerPawn NewPawn))
  (bind _currentminigame (Variables|/Script/FortniteGame.FortMinigameLogicComponent|GetCurrentMinigame (Variables|Default|GetFortMinigameProgress))
    (:then
      (if (Minigame|IsinProgress _currentminigame)
        (EventDispatchers|BindEventtoOnDied _asfort_player_pawn (delegate Custom|OnPlayerPawnDied))
        (bind _asfort_player_state_athena (Utilities|Casting|CastToFortPlayerStateAthena (Variables|/Script/Engine.Controller|GetPlayerState (Pawn|GetController _asfort_player_pawn))))
        (if (CallFunction|IsActoronSameTeam _asfort_player_pawn)
          (Utilities|ShouldGrantItemstoPlayer (PlayerState|GetPlayerUniqueNetID _asfort_player_state_athena) (Pawn|GetController _asfort_player_pawn)
            (:true
              (bind _asfort_inventory_owner_interface (Utilities|Casting|CastToFortInventoryOwnerInterface (Pawn|GetController _asfort_player_pawn)))
              (CallFunction|TryGiveAllItemstoPlayer _asfort_inventory_owner_interface)
              (Utilities|Set|Add (Variables|Default|GetGrantedPlayers) (PlayerState|GetPlayerUniqueNetID _asfort_player_state_athena))))
          (else
            (Utilities|Set|Remove (Variables|Default|GetGrantedPlayers) (PlayerState|GetPlayerUniqueNetID _asfort_player_state_athena))))))
    (:else)))

(event EventDestroyed
  (bind _minigame (TeamSettings|GetCurrentMinigame))
  ;[1 0.06933 0 0.5] Unbind events during cleanup
  (EventDispatchers|UnbindEventfromOnItemAdded (Variables|Default|GetPickupItemList) (EventDispatchers|CreateEvent :!event Custom|OnAddedToInventoryList))
  (Default|UnbindEventfromOnItemListUpdated (Variables|Default|GetPickupItemList) (EventDispatchers|CreateEvent :!event Custom|OnItemListUpdated))
  (Teams|RefreshTeamChangeBinding 0 true (EventDispatchers|CreateEvent :!event Update%20Colors))
  (CallFunction|BindToSpawning true (TeamSettings|GetCurrentMinigame))
  (Utilities|IsValid _minigame
    (:Is%20Valid
      (Minigame|UnbindEventfromPlayerAddedDelegate _minigame (EventDispatchers|CreateEvent :!event Custom|OnPlayerAddedToMinigame))
      (CallFunction|RemoveAllPlayerBinds _minigame))))

(event Custom|OnPlayerPawnDied
  (params (DamagedActor :type Actor%20Object%20Reference) (Damage :type Float%20%28single-precision%29) (InstigatedBy :type Controller%20Object%20Reference) (DamageCauser :type Actor%20Object%20Reference) (HitLocation :type Vector) (FHitComponent :type Primitive%20Component%20Object%20Reference) (BoneName :type Name) (Momentum :type Vector))
  (bind _returnvalue (Team|GetActorTeam InstigatedBy))
  ;[0.4836 0.6408 1] On team elminiation send a trigger message
  (bind _pawn (Utilities|IsPlayerinRunningMinigame DamagedActor)
    (:%20%20
      (Utilities|FlowControl|Sequence
        (:then_0
          (if (and (not (Variables|Default|GetbTeamOutOfRespawnsSignalTriggered)) (CallFunction|IsTeamOutOfRespawns))
            (GameplayEvent|BroadcastEvent (Variables|UserOptions|Events|GetOnTeamIsOutOfRespawns) 0 (Utilities|Struct|MakeGameplayEventTeamEliminated 0 (Team|GetActorTeam (Pawn|GetController _pawn)) _returnvalue))
            (Variables|Default|SetbTeamOutOfRespawnsSignalTriggered true)))
        (:then_1
          (if (CallFunction|IsActoronSameTeam _pawn)
            (GameplayEvent|BroadcastEvent (Variables|UserOptions|Events|GetOnTeamMemberEliminated) 0 (Utilities|Struct|MakeGameplayEventTeamMemberEliminated (Pawn|GetController _pawn) (Team|GetActorTeam (Pawn|GetController _pawn)) _returnvalue))))
        (:then_2
          (if (and (CallFunction|IsActoronSameTeam InstigatedBy) (Utilities|Enum|NotEqual(Enum) (Team|GetTeamAffiliation DamagedActor InstigatedBy) "Friendly"))
            (GameplayEvent|BroadcastEvent (Variables|UserOptions|Events|GetOnEnemyEliminatedByTeamMember) 0 (Utilities|Struct|MakeGameplayEventPlayerEliminatedbyTeamMember InstigatedBy (Team|GetActorTeam (Pawn|GetController _pawn)) _returnvalue))))))))

(event Custom|Bind%20and%20Handle%20Drop%20Box%20Collision
  (params (Minigame :type Fort%20Minigame%20Object%20Reference))
  ;[1 0.1082 0 0.25] Disable Collision on drop area during Edit mode, as overlaps are expensive.
  (Minigame|BindEventtoOnPlayModeChangedDelegate Minigame (delegate Custom|OnPlayModeChanged))
  (CallFunction|OnPlayModeChanged Minigame (Minigame|IsinPlayMode Minigame :!class /Script/FortniteGame.FortMinigame)))

(event Custom|OnPlayModeChanged
  (params (Minigame :type Fort%20Minigame%20Object%20Reference) (bIsInPlayMode :type Boolean))
  ;[1 0.1082 0 0.25] Disable Collision on drop area during Edit mode, as overlaps are expensive.
  (Collision|SetCollisionEnabled (Variables|Default|GetDropBox) (select bIsInPlayMode "NoCollision" "QueryOnly")))

(event Custom|OnAddedToInventoryList
  (params (AddedActor :type Actor%20Object%20Reference))
  (CallFunction|AddItemtoStartingInventory AddedActor)
  (CallFunction|ResaveActor))

(event OnRemovedfromMinigame(FortMinigameProgress)
  (params (Minigame :type Fort%20Minigame%20Object%20Reference))
  ;[1 0.3598 0.1485] Unbind events when removed from minigame
  (Utilities|IsValid Minigame
    (:Is%20Valid
      (CallFunction|BindToSpawning true Minigame)
      (Minigame|UnbindEventfromOnPawnTeleportedforMinigame Minigame (EventDispatchers|CreateEvent :!event Custom|InitialInventorySetup))
      (Minigame|UnbindEventfromPlayerAddedDelegate Minigame (EventDispatchers|CreateEvent :!event Custom|OnPlayerAddedToMinigame))
      (Minigame|UnbindEventfromPlayerRemovedDelegate Minigame (EventDispatchers|CreateEvent :!event Custom|OnPlayerRemoved))
      (TeamSettings|RemoveTeamSettingsfromMinigame Minigame)
      (Teams|RefreshTeamChangeBinding 0 true (EventDispatchers|CreateEvent :!event Update%20Colors))
      (CallFunction|RemoveAllPlayerBinds Minigame))))

(event Custom|OnItemListUpdated
  (if (== (ItemData|GetNumItems (Variables|Default|GetPickupItemList)) 0)
    (CallFunction|ClearItems)))

(event OnComponentHit(DropBox)
  (params (HitComponent :type Primitive%20Component%20Object%20Reference) (OtherActor :type Actor%20Object%20Reference) (OtherComp :type Primitive%20Component%20Object%20Reference) (NormalImpulse :type Vector :default "0, 0, 0") (Hit :type Hit%20Result%20Structure%20%28by%20ref%29))
  ;[1 1 1] Handles item being added from a hit event.  Hit captures all pick-ups
  (Networking|SwitchHasAuthority
    (:Authority
      (bind _asfort_pickup_athena (Utilities|Casting|CastToFortPickupAthena OtherActor))
      (Utilities|IsPickupValid _asfort_pickup_athena (Variables|Default|GetPickupValidGameplayTagContainer)
        (:IsValid
          (CallFunction|AddItemtoQueue _asfort_pickup_athena))))))

(event OnAddedtoMinigame(FortMinigameProgress)
  (params (Minigame :type Fort%20Minigame%20Object%20Reference))
  ;[0.2708 1 0.3543] Gives items if join in progress is set to TRUE
  (Utilities|IsValid Minigame
    (:Is%20Valid
      (CallFunction|BindToSpawning false Minigame)
      (CallFunction|SetupRespawnItemGrantBinds Minigame)
      (Minigame|BindEventtoOnPawnTeleportedforMinigame Minigame (delegate Custom|InitialInventorySetup))
      (Minigame|BindEventtoPlayerRemovedDelegate Minigame (EventDispatchers|CreateEvent :!event Custom|OnPlayerRemoved))
      (TeamSettings|AddTeamSettingstoMinigame Minigame)
      (Teams|RefreshTeamChangeBinding 0 false (EventDispatchers|CreateEvent :!event Update%20Colors))
      (CallFunction|BindAndHandleDropBoxCollision Minigame))))

(event Custom|InitialInventorySetup
  (params (PlayerPawn :type Fort%20Player%20Pawn%20Object%20Reference))
  (bind _currentminigame (Utilities|GetMinigame))
  (bind _returnvalue (Minigame|GetMinigameTeamBalanceComponent _currentminigame))
  (bind _returnvalue_2 (Pawn|GetController PlayerPawn))
  ;[0.2708 1 0.3543] Gives items if join in progress is set to TRUE
  (bind _isinsameisland (CallFunction|IsPlayeronSameIsland PlayerPawn))
  (if _isinsameisland
    (bind _currentminigame_2 (Variables|/Script/FortniteGame.FortMinigameLogicComponent|GetCurrentMinigame (Variables|Default|GetFortMinigameProgress))
      (:then
        (EventDispatchers|BindEventtoOnDied PlayerPawn (EventDispatchers|CreateEvent :!event Custom|OnPlayerPawnDied))
        (if (and (Minigame|IsinProgress _currentminigame_2) (CallFunction|IsActoronSameTeam PlayerPawn))
          (bind _asfort_player_state_athena (Utilities|Casting|CastToFortPlayerStateAthena (Variables|/Script/Engine.Controller|GetPlayerState _returnvalue_2)))
          (Utilities|IsValid _returnvalue
            (:Is%20Valid
              (if (not (and (not (Minigame|Teams|ShouldRotateTeams _returnvalue)) (Variables|/Script/FortniteGame.FortMinigame|GetbKeepItemsBetweenRounds _currentminigame)))
                (bind _asfort_inventory_owner_interface (Utilities|Casting|CastToFortInventoryOwnerInterface _returnvalue_2))
                (CallFunction|TryGiveAllItemstoPlayer _asfort_inventory_owner_interface)
                (Utilities|Set|Add (Variables|Default|GetGrantedPlayers) (PlayerState|GetPlayerUniqueNetID _asfort_player_state_athena)))))))
      (:else))))

(event Custom|UpdateTeamAssignmentDisplay
  (bind _score_mid (Variables|DeviceVisuals|GetScore%20MID))
  ;[1 1 1] Updates the display of which team the toy  is assigned to
  (Networking|CanExecuteCosmeticEvents
    (:True
      (Material|ReattachOrCreateMID (Variables|Default|GetCreativeEditOnlyMesh) _score_mid 1)
      (Rendering|Material|SetScalarParameterValue _score_mid "Score" (select (Creative|TeamOption|IsTeamType (Variables|UserOptions|General|GetTeamSetting) "All") 0.0 (Math|Conversions|ToFloat(Integer) (Creative|TeamOption|GetLegacyTeamIndex (Variables|UserOptions|General|GetTeamSetting) false))) :!class /Script/Engine.MaterialInstanceDynamic))))

(event EventOnMinigameEnded
  (Networking|SwitchHasAuthority
    (:Authority
      (Minigame|UnbindEventfromPlayerChangedTeamDelegate (Utilities|GetMinigame) (delegate Custom|On%20Player%20Changed%20Team))
      (Utilities|Set|Clear (Variables|Default|GetGrantedPlayers))
      (GameplayEventRouter|StopListeningtoEvent (Variables|Default|GetTeamChangedObserver)))))

(event EventOnMinigameStarted
  (Networking|SwitchHasAuthority
    (:Authority
      (Minigame|BindEventtoPlayerChangedTeamDelegate (Utilities|GetMinigame) (delegate Custom|On%20Player%20Changed%20Team))
      (Variables|Default|SetbTeamOutOfRespawnsSignalTriggered false)
      (bind _returnvalue (GameplayEventRouter|GetEventRouter "/Script/FortniteGame.FortPlayspace" self)) ; ContextActor is how we bind to a specific playspace
      (bind (_asynctaskproxy _payload) (GameplayEventRouter|StartListeningToEvent _returnvalue "/Script/FortniteGame.FortTeamChangedContext")
        (:then
          (Variables|Default|SetTeamChangedObserver _asynctaskproxy))
        (:OnEventReceived
          (bind (_payload_player _payload_teamid) (Utilities|Struct|BreakFortTeamChangedContext _payload))
          (CallFunction|OnPlayerChangedTeam _payload_player _payload_teamid 0 true) ; TeamId and OldTeamId don't get used, but fill out what we can anyway
        )) ; Context here is irrelevant
    )))

(event Custom|EndRoundFromEvent
  (params (Instigator :type Controller%20Object%20Reference))
  (bind _currentminigame (Variables|/Script/FortniteGame.FortMinigameLogicComponent|GetCurrentMinigame (Variables|Default|GetFortMinigameProgress))
    (:then
      (if (Creative|TeamOption|IsTeamType (Variables|UserOptions|General|GetTeamSetting))
        (Minigame|SetWinningTeamIndex _currentminigame (Creative|TeamOption|ConverttoTeam (Variables|UserOptions|General|GetTeamSetting)))
        (else
          _))
      (Minigame|EndGame _currentminigame "EndRound"))
    (:else)))

(event Custom|On%20Player%20Changed%20Team
  (params (PlayerState :type Fort%20Player%20State%20Object%20Reference) (TeamId :type Byte) (OldTeamId :type Byte) (bTeamChangedWithoutRespawn :type Boolean))
  (bind _returnvalue (Class|FortPlayerState|FindFNEPlayerComponent PlayerState))
  (bind _returnvalue_2 (PlayerState|GetCurrentPawn PlayerState))
  (CreativeTeam|ActorMatchesTeam PlayerState (Variables|UserOptions|General|GetTeamSetting)
    (:Yes
      (Utilities|Set|Remove (Variables|Default|GetGrantedPlayers) (PlayerState|GetPlayerUniqueNetID PlayerState))
      (if bTeamChangedWithoutRespawn
        ;[1 1 1] This is to support team switching without respawn otherwise this would be handled by the OnPawnTeleportedForMinigame event.
        (Utilities|IsValid _returnvalue_2
          (:Is%20Valid
            (CallFunction|InitialInventorySetup _returnvalue_2)
            (Utilities|IsValid _returnvalue
              (:Is%20Valid
                (Class|FNEPlayerComponent|ApplyMutatorstoPlayer _returnvalue false)))))))))

(event Custom|SwitchPlayerToSpectate
  (params (Instigator :type Controller%20Object%20Reference))
  (if (CallFunction|IsActoronSameTeam Instigator)
    (bind _returnvalue (Class|PlayspaceComponentSpectateMode|GetPlayspaceComponent_SpectateMode self))
    (if (and (Utilities|IsValid _returnvalue :!class /Script/Engine.KismetSystemLibrary) (Utilities|IsValid Instigator :!class /Script/Engine.KismetSystemLibrary))
      (Class|PlayspaceComponentSpectateMode|ForceStartSpectating _returnvalue Instigator))))

(event Custom|OnPlayerRemoved
  (params (UniqueNetId :type Unique%20Net%20Id%20Structure) (IsLocalPlayer :type Boolean))
  (bind _returnvalue (Fortnite|Athena|GameState|GetGameStateAthena))
  (bind _returnvalue_2 (Class|FortGameState|GetFortPlayerStatefromUniqueNetId _returnvalue UniqueNetId))
  (bind _returnvalue_3 (Fortnite|GetControllerfromActor _returnvalue_2))
  (bind _returnvalue_4 (PlayerState|GetCurrentPawn _returnvalue_2))
  ;[1 0.2636 0.03237] Stop tracking players that have been removed from the minigame
  (Utilities|Set|Remove (Variables|Default|GetGrantedPlayers) UniqueNetId)
  (Development|PrintString "OnPlayerRemoved")
  (Utilities|IsValid _returnvalue
    (:Is%20Valid
      (Utilities|IsValid _returnvalue_2
        (:Is%20Valid
          (Utilities|FlowControl|Sequence
            (:then_0
              (Utilities|IsValid _returnvalue_3
                (:Is%20Valid
                  (Pawn|UnbindEventfromOnPossessedPawnChanged _returnvalue_3 (EventDispatchers|CreateEvent :!event Custom|OnPossessedPawnChanged)))))
            (:then_1
              (Utilities|IsValid _returnvalue_4
                (:Is%20Valid
                  (EventDispatchers|UnbindEventfromOnDied _returnvalue_4 (EventDispatchers|CreateEvent :!event Custom|OnPlayerPawnDied)))))))))))

(event OnItemListUpdated(PickupItemList)
  (if (== (ItemData|GetNumItems (Variables|Default|GetPickupItemList)) 0)
    (CallFunction|ClearItems)))

(event Custom|RespawnAtPlayerSpawner%20Event
  (params (Instigator :type Player%20Controller%20Object%20Reference))
  (bind _minigame (TeamSettings|GetCurrentMinigame))
  ;[0.4042 1 0.6378] Respawn At Player Spawner Function
  (bind _asfort_player_state (Utilities|Casting|CastToFortPlayerState (Variables|/Script/Engine.Controller|GetPlayerState Instigator)))
  (bind _asfort_player_controller_athena (Utilities|Casting|CastToFortPlayerControllerAthena (Fortnite|GetFortPlayerControllerfromActor _asfort_player_state)))
  (if (Variables|UserOptions|Mode|Spawning|GetbRespawnAlivePlayers)
    (FortPawn|ForceKill :self (Game|Player|GetPlayerPawn _asfort_player_controller_athena))
    ;;[1 1 1] if we don't respawn manually afterwards, player gets stuck at loading screen
    (if (Minigame|IsinDoNotSpawnMode _minigame)
      (Class|FortMinigamePlayerSpawnComponent|SetRespawnDataandRespawn (Minigame|GetPlayerSpawnComponent _minigame) _asfort_player_state))
    (else
      (Class|FortPlayerControllerAthena|RespawnPlayerAfterDeath _asfort_player_controller_athena "(TagName=\"\")" false false))))

(event MinigameStateChangedDelegate(FortMinigameProgress)
  (params (Minigame :type Fort%20Minigame%20Object%20Reference) (NewMinigameState :type EFortMinigameState%20Enum :default "PreGame"))
  ;[1 1 1] Update team settings to make sure that they were not overwritten by another instance of the device during island editing
  (switch Utilities|FlowControl|Switch|SwitchonEFortMinigameState NewMinigameState
    (:Setup
      _)
    (:InProgress
      _))
  (TeamSettings|AddTeamSettingstoMinigame (TeamSettings|GetCurrentMinigame)))

(event HandleVersionUpgrades(CreativeVersion)
  (params (LastSavedVersion :type Creative%20Version%20Id%20Structure))
  (UserOptions|HandleReleaseVersionUpgrades LastSavedVersion))

(event Custom|RPC_MULTICAST%20Add%20Item%20FX
  (params (Item :type Actor%20Object%20Reference))
  (Networking|CanExecuteCosmeticEvents
    (:True
      (bind _asfort_pickup_athena (Utilities|Casting|CastToFortPickupAthena Item))
      (Class|CreativeItemPreviewComponent|ItemAddedToQueueFX (Variables|Default|GetCreative_ItemPreview_Component) _asfort_pickup_athena))))

(fn AddItemToQueue
  (params (NewItem :type Fort%20Pickup%20Athena%20Object%20Reference))
  (bind _banyitemblacklisted (CallFunction|IsAnyItemBlacklisted (Variables|GetNewItem)))
  (if (not _banyitemblacklisted)
    (bind _bitemadded (Class|PickupItemListComponent|AddItemtoQueue (Variables|Default|GetPickupItemList) (Variables|GetNewItem)))
    (if _bitemadded
      (CallFunction|RPCMULTICASTAddItemFX (Variables|GetNewItem)))))

(fn AddItemToStartingInventory
  (params (AddedActor :type Actor%20Object%20Reference))
  (Networking|FlushNetDormancy))

(fn AddTeamSettingsToMinigame
  (params (Minigame :type Fort%20Minigame%20Object%20Reference))
  (bind (_invincibility _startinghealth _maxhealth _startingshield _maxshields _dbnotype _nocooldowns _noswapcooldowns _infinitemagazineammo _infinitedurability _infinitecharges _infinitereserveenergy _infiniteloadedenergy _infinitebuildingresources _infinitegold _infiniteworldresources _displayemptyammoslots _movementmultiplier _applyjumppenalty _gliderredeploy _flyenabled) (Class|GetClassDefaults "/FortniteCoreUI/Settings/ExperienceSettingsComponent.ExperienceSettingsComponent_C"))
  (bind (_overridespawnimmunitytime _spawnimmunitytime _maxitemslots) (Class|GetClassDefaults "/CreativeCoreDevices/ExperienceSettings/ExperienceSettingsComponent_Minigame.ExperienceSettingsComponent_Minigame_C"))
  (bind (_mantlingheightwaterparams _mantlingheightwatersetting) (Class|GetClassDefaults "/CreativeCoreDevices/ExperienceSettings/MantlingConversionConstants.MantlingConversionConstants_C"))
  (bind (_mantlingheightparams _mantlingheightsetting) (Class|GetClassDefaults "/CreativeCoreDevices/ExperienceSettings/MantlingConversionConstants.MantlingConversionConstants_C"))
  (Networking|SwitchHasAuthority
    (:Authority
      (Utilities|FlowControl|Sequence
        (:then_0
          (LocalVariables|SetCurrentTeamIndex (select (Creative|TeamOption|IsTeamType (Variables|UserOptions|General|GetTeamSetting) "All") 255 (Math|Byte|MakeLiteralByte (Creative|TeamOption|ConverttoTeam (Variables|UserOptions|General|GetTeamSetting))))))
        (:then_1
          (bind _output_get (LocalVariables|SetTeamSettingsMutator (TeamSettings|GetTeamSettingsMutator (Variables|GetMinigame))))
          (Utilities|IsValid _output_get
            (:Is%20Valid
              (bind _returnvalue (TeamSettings|AddDataSourceforTeam _output_get (LocalVariables|GetCurrentTeamIndex) (Variables|Default|GetToyOptionsComponent)))
              (if _returnvalue
                (Minigame|GameRules|RefreshGameRulesPreview (Variables|GetMinigame))))))
        (:then_2
          (bind _output_get_2 (LocalVariables|SetWeaponSettingsMutator (TeamSettings|GetWeaponSettingsMutator (Variables|GetMinigame))))
          (Utilities|IsValid _output_get_2
            (:Is%20Valid
              (Class|FortAthenaMutatorWeaponSettings|AddDataSourceforTeam _output_get_2 "/Game/Athena/Items/Cosmetics/Pickaxes/DefaultPickaxe.DefaultPickaxe" (LocalVariables|GetCurrentTeamIndex) (Variables|Default|GetToyOptionsComponent)))))
        (:then_3
          _)))
    (:Remote
      _))
  ; ApplySharedSettings
  (Utilities|FlowControl|Sequence
    (:then_0
      (Class|FortAthenaMutatorUtilities|RemoveDataRegistryOptionsforContextObject)
      (Class|FortAthenaMutatorUtilities|ApplyActorOptionsComponenttoDataRegistry (Variables|Default|GetToyOptionsComponent)))
    (:then_1
      (ScalableFloat|SetValuewithContext _mantlingheightsetting (Class|MantlingConversionHelpers|ConvertMantlingHeightPresettoFloat (Variables|UserOptions|Player|Locomotion|GetMantlingMinimumHeight) _mantlingheightparams) (Variables|UserOptions|Overrides|GetMantlingMinimumHeight_Override))
      (ScalableFloat|SetValuewithContext _mantlingheightwatersetting (Class|MantlingConversionHelpers|ConvertMantlingHeightPresettoFloat (Variables|UserOptions|Player|Locomotion|GetMantlingMinimumHeightInWater) _mantlingheightwaterparams) (Variables|UserOptions|Overrides|GetMantlingMinimumHeightInWater_Override))
      ;; UpdateCoreExperienceSettings
      (Utilities|FlowControl|Sequence
        (:then_0
          (ScalableFloat|SetValuewithContext _invincibility (Variables|UserOptions|Player|Health|GetbInvincibility) (Variables|UserOptions|Overrides|GetbInvincibility_Override))
          (ScalableFloat|SetValuewithContext _maxhealth (Variables|UserOptions|Player|Health|GetMaxHealthAlt) (Variables|UserOptions|Overrides|GetMaxHealthAlt_Override))
          (ScalableFloat|SetValuewithContext _startinghealth (Variables|UserOptions|Player|Health|GetStartingHealthPercentage) (Variables|UserOptions|Overrides|GetStartingHealthPercentage_Override))
          (ScalableFloat|SetValuewithContext _maxshields (Variables|UserOptions|Player|Shields|GetMaxShields) (Variables|UserOptions|Overrides|GetMaxShields_Override))
          (ScalableFloat|SetValuewithContext _startingshield (Variables|UserOptions|Player|Shields|GetStartingShieldPercentage) (Variables|UserOptions|Overrides|GetStartingShieldPercentage_Override))
          (ScalableFloat|SetValuewithContext _dbnotype (Variables|UserOptions|Mode|Eliminations|GetDownButNotOut) (Variables|UserOptions|Overrides|GetDownButNotOut_Override))
          (ScalableFloat|SetValuewithContext _movementmultiplier (Variables|UserOptions|Player|Locomotion|GetMovementMultiplier) (Variables|UserOptions|Overrides|GetMovementMultiplier_Override))
          (ScalableFloat|SetValuewithContext _nocooldowns (Variables|UserOptions|Player|Inventory|GetbNoCooldowns) (Variables|UserOptions|Overrides|GetbNoCooldowns_Override))
          (ScalableFloat|SetValuewithContext _noswapcooldowns (Variables|UserOptions|Player|Inventory|GetbNoCooldownsAfterSwap) (Variables|UserOptions|Overrides|GetbNoCooldownsAfterSwap_Override))
          (ScalableFloat|SetValuewithContext _infinitemagazineammo (Variables|UserOptions|Player|Inventory|GetbInfiniteMagazineAmmo) (Variables|UserOptions|Overrides|GetbInfiniteMagazineAmmo_Override))
          (ScalableFloat|SetValuewithContext _infinitedurability (Variables|UserOptions|Player|Inventory|GetbInfiniteDurability) (Variables|UserOptions|Overrides|GetbInfiniteDurability_Override))
          (ScalableFloat|SetValuewithContext _infinitecharges (Variables|UserOptions|Player|Inventory|GetbInfiniteCharges) (Variables|UserOptions|Overrides|GetbInfiniteCharges_Override))
          (ScalableFloat|SetValuewithContext _infinitereserveenergy (Variables|UserOptions|Player|Inventory|GetbInfiniteReserveEnergy) (Variables|UserOptions|Overrides|GetbInfiniteReserveEnergy_Override))
          (ScalableFloat|SetValuewithContext _infiniteloadedenergy (Variables|UserOptions|Player|Inventory|GetbInfiniteLoadedEnergy) (Variables|UserOptions|Overrides|GetbInfiniteLoadedEnergy_Override))
          (ScalableFloat|SetValuewithContext _infinitebuildingresources (Variables|UserOptions|Player|Inventory|GetbInfiniteBuildingResources) (Variables|UserOptions|Overrides|GetbInfiniteBuildingResources_Override))
          (ScalableFloat|SetValuewithContext _infinitegold (Variables|UserOptions|Player|Inventory|GetbInfiniteGold) (Variables|UserOptions|Overrides|GetbInfiniteGold_Override))
          (ScalableFloat|SetValuewithContext _infiniteworldresources (Variables|UserOptions|Player|Inventory|GetbInfiniteWorldResources) (Variables|UserOptions|Overrides|GetbInfiniteWorldResources_Override))
          (ScalableFloat|SetValuewithContext _applyjumppenalty (Variables|UserOptions|Player|Locomotion|GetbApplyJumpingPenalty) (Variables|UserOptions|Overrides|GetbApplyJumpingPenalty_Override))
          (ScalableFloat|SetValuewithContext _gliderredeploy (Variables|UserOptions|Player|Locomotion|GetbGliderRedeployable) (Variables|UserOptions|Overrides|GetbGliderRedeployable_Override))
          (ScalableFloat|SetValuewithContext _flyenabled (Variables|UserOptions|Player|Locomotion|GetbFlyingEnabled) (Variables|UserOptions|Overrides|GetbFlyingEnabled_Override))
          (ScalableFloat|SetValuewithContext _displayemptyammoslots (Variables|UserOptions|Player|Inventory|GetbDisplayEmptyAmmoSlots) (Variables|UserOptions|Overrides|GetbDisplayEmptyAmmoSlots_Override))
          (ScalableFloat|SetValuewithContext (Variables|Default|GetScalableFloat_AllowImpulsingTeammates) (Variables|UserOptions|Mode|TeamSettings|GetbAllowImpulsingTeammates)))
        (:then_1
          _))
      ;; UpdateInGameExperienceSettings
      (Utilities|FlowControl|Sequence
        (:then_0
          (ScalableFloat|SetValuewithContext _overridespawnimmunitytime (Variables|UserOptions|Mode|Spawning|GetbOverrideSpawnImmunityTime) (Variables|UserOptions|Overrides|GetbOverrideSpawnImmunityTime_Override))
          (ScalableFloat|SetValuewithContext _spawnimmunitytime (Variables|UserOptions|Mode|Spawning|GetSpawnImmunityTime) (Variables|UserOptions|Overrides|GetSpawnImmunityTime_Override))
          (ScalableFloat|SetValuewithContext _maxitemslots (Variables|UserOptions|Player|Equipment|GetMaximumEquipmentSlots) (Variables|UserOptions|Overrides|GetMaximumEquipmentSlots_Override)))
        (:then_1
          (if (not (Variables|UserOptions|Player|Locomotion|GetbFallDamage))
            (ScalableFloat|SetValuewithContext (Class|GetClassDefaults "/CreativeCoreDevices/ExperienceSettings/ExperienceSettingsComponent_Minigame.ExperienceSettingsComponent_Minigame_C") false)))
        (:then_2
          _)))))

(fn ApplyColors
  (params (Color :type Linear%20Color%20Structure))
  (bind _ring_mid (Variables|DeviceVisuals|GetRing%20MID))
  (Material|ReattachOrCreateMID (Variables|Default|GetCreativeEditOnlyMesh) _ring_mid)
  (Rendering|Material|SetVectorParameterValue _ring_mid "MultiplyColor" (Math|Color|Desaturate(LinearColor) (Math|Color|LinearColor*Float Color (Variables|Default|GetTeam%20Color%20Emissive%20Intensity)) -0.1) :!class /Script/Engine.MaterialInstanceDynamic)
  (bind _output_get (Variables|Default|SetActiveColor (Variables|GetColor)))
  (Rendering|Material|SetVectorParameterValue (Variables|DeviceVisuals|GetRing%20MID) "ColorMain" _output_get :!class /Script/Engine.MaterialInstanceDynamic)
  (Maerial|SetColor (Variables|Default|GetCreativeEditOnlyBase) _output_get _output_get)
  (Material|SetColor (Variables|Default|GetCreativeEditOnlyBeam) (Math|Color|LinearColor*Float _output_get 0.3)))

(fn BindtoSpawning
  (params (Force%20Unbind :type Boolean) (Minigame :type Fort%20Minigame%20Object%20Reference))
  (bind _outputdelegate (EventDispatchers|CreateEvent :!event On%20AI%20Spawned))
  (bind _outputdelegate_2 (EventDispatchers|CreateEvent :!event On%20Player%20Spawned))
  (bind _minigame (Variables|GetMinigame))
  (Networking|SwitchHasAuthority
    (:Authority
      (Utilities|FlowControl|Sequence
        (:then_0
          (bind _returnvalue (AI|GetAISpawner))
          (Utilities|IsValid _returnvalue
            (:Is%20Valid
              (if (and (Variables|UserOptions|Mode|Spawning|GetbSpawnEventActivatesForAI) (not Force%20Unbind))
                (EventDispatchers|BindEventtoOnPawnSpawned _returnvalue _outputdelegate)
                (else
                  (EventDispatchers|UnbindEventfromOnPawnSpawned _returnvalue _outputdelegate))))))
        (:then_1
          (Utilities|IsValid _minigame
            (:Is%20Valid
              (if (and (Variables|UserOptions|Mode|Spawning|GetbSpawnEventActivatesForPlayers) (not Force%20Unbind))
                (Minigame|BindEventtoOnMinigamePlayerSpawned _minigame _outputdelegate_2)
                (else
                  (Minigame|UnbindEventfromOnMinigamePlayerSpawned _minigame _outputdelegate_2))))))))))

(fn BuildDataRegistryResolverScope
  (bind _teamsetting (Variables|UserOptions|General|GetTeamSetting))
  (Utilities|FlowControl|Sequence
    (:then_0
      (bind _output_get (LocalVariables|SetSpatialInterfaceName (SpatialGameplay|GetSpatialGameplayInterfaceFNameforActor self)))
      (if (!= _output_get (LocalVariables|GetNameNone))
        (Utilities|Array|Add (LocalVariables|GetOutResolverScopes) _output_get)))
    (:then_1
      (if (not (Creative|TeamOption|IsTeamType _teamsetting "All"))
        (Utilities|Array|Add (LocalVariables|GetOutResolverScopes) (Utilities|String|StringToName (Utilities|String|BuildString(Integer) "Team" "" (Math|Conversions|ToInteger(Byte) (Math|Byte|MakeLiteralByte (Creative|TeamOption|ConverttoTeam _teamsetting))))))))
    (:then_2
      (return (LocalVariables|GetOutResolverScopes) 0 (> (Utilities|Array|Length (LocalVariables|GetOutResolverScopes)) 0)))))

(fn ClearItems
  (Networking|SwitchHasAuthority
    (:Authority
      (Networking|FlushNetDormancy)
      (CallFunction|ResaveActor))))

(fn GetAllowTemporaryTeamMemberInvitations_Override_Verse
  (if (Variables|Verse|GetAllowTemporaryTeamMemberInvitations_VerseOverride_IsSet)
    (return (Variables|Verse|GetAllowTemporaryTeamMemberInvitations_VerseOverride))
    (else
      (return (Variables|/Script/Creative_TemporaryTeamsRuntime.FortUserOptionComponent_TemporaryTeams_TeamSettingsAndInventory|GetbAllowTemporaryTeamInvites_Override (Variables|Default|GetBP_TemporaryTeams_TeamSettingsAndInventory))))))

(fn GetAllowTemporaryTeamMemberInvitations_Verse
  (if (Variables|Verse|GetAllowTemporaryTeamMemberInvitations_VerseOverride_IsSet)
    (return (Variables|Verse|GetAllowTemporaryTeamMemberInvitations_VerseOverride))
    (else
      (return (Variables|/Script/Creative_TemporaryTeamsRuntime.FortUserOptionComponent_TemporaryTeams_TeamSettingsAndInventory|GetbAllowTemporaryTeamInvites (Variables|Default|GetBP_TemporaryTeams_TeamSettingsAndInventory))))))

(fn GetCacheContextOverride
  (return 0))

(fn GetCurrentMinigame
  (return (Variables|/Script/FortniteGame.FortMinigameLogicComponent|GetCurrentMinigame (Variables|Default|GetFortMinigameProgress))))

(fn GetTeamMembers
  (bind _currentminigame (Variables|/Script/FortniteGame.FortMinigameLogicComponent|GetCurrentMinigame (Variables|Default|GetFortMinigameProgress))
    (:then
      (bind _outplayers (Minigame|GetParticipatingPlayers_NotPure _currentminigame))
      (for _array_element _outplayers
        (if (CallFunction|IsActoronSameTeam _array_element)
          (Utilities|Array|Add (LocalVariables|GetTeamControllers) (Fortnite|GetControllerfromActor _array_element)))))
    (:else
      _))
  (return (LocalVariables|GetTeamControllers)))

(fn GetTeamSettingsMutator
  (params (Minigame :type Fort%20Minigame%20Object%20Reference))
  (Utilities|FlowControl|Sequence
    (:then_0
      (bind _output_get (LocalVariables|SetCurrentMinigame Minigame))
      (Utilities|IsValid _output_get
        (:Is%20Valid
          (bind _output_get_2 (LocalVariables|SetMutatorList (Actor|GetComponentbyClass _output_get "/Script/FortniteGame.FortMutatorListComponent")))
          (Utilities|IsValid _output_get_2
            (:Is%20Valid
              (bind _asfort_athena_mutator_player_team_settings (Utilities|Casting|CastToFortAthenaMutator_PlayerTeamSettings (Modifier|GetMutatorbyClass _output_get_2 "/Script/FortniteGame.FortAthenaMutator_PlayerTeamSettings")))
              (LocalVariables|SetTeamSettingsMutator _asfort_athena_mutator_player_team_settings))))))
    (:then_1
      (return (LocalVariables|GetTeamSettingsMutator)))))

(fn GetWeaponSettingsMutator
  (params (Minigame :type Fort%20Minigame%20Object%20Reference))
  (bind _minigame (Variables|GetMinigame))
  (bind _volume (Variables|/Script/FortniteGame.FortMinigame|GetVolume _minigame))
  (Utilities|FlowControl|Sequence
    (:then_0
      (Utilities|IsValid _minigame
        (:Is%20Valid
          (Utilities|IsValid _volume
            (:Is%20Valid
              (bind _output_get (LocalVariables|SetCreativeIslandMutatorList (Actor|GetComponentbyClass _volume "/Game/Athena/Playlists/Creative/Mutators/CreativeIslandMutatorList.CreativeIslandMutatorList_C")))
              (Utilities|IsValid _output_get
                (:Is%20Valid
                  (bind _asfort_athena_mutator_weapon_settings (Utilities|Casting|CastToFortAthenaMutator_WeaponSettings (Modifier|GetMutatorbyClass _output_get "/Script/FortniteGame.FortAthenaMutator_WeaponSettings")))
                  (LocalVariables|SetOutMutator _asfort_athena_mutator_weapon_settings))))))))
    (:then_1
      (return (LocalVariables|GetOutMutator)))))

(fn HandleReleaseVersionUpgrades
  (params (Last%20Saved%20Version :type Creative%20Version%20Id%20Structure%20%28by%20ref%29))
  (Utilities|FlowControl|Sequence
    (:then_0
      (CreativeVersion|CompareVersiontoTag (Variables|GetLast%20Saved%20Version) "(TagName=\"Creative.Version.Release.32.00\")"
        (:OlderVersion
          (bind _output_get (Variables|UserOptions|Player|Inventory|SetbInfiniteCharges (Variables|UserOptions|Player|Inventory|GetbInfiniteAmmo)))
          (Variables|UserOptions|Player|Inventory|SetbInfiniteReserveEnergy _output_get)
          (bind _output_get_2 (Variables|UserOptions|Overrides|SetbInfiniteCharges_Override (Variables|UserOptions|Overrides|GetbInfiniteAmmo_Override)))
          (Variables|UserOptions|Overrides|SetbInfiniteReserveEnergy_Override _output_get_2)
          (LocalVariables|SetbWasMigrated true))))
    (:then_1
      (CreativeVersion|CompareVersiontoTag (Variables|GetLast%20Saved%20Version) "(TagName=\"Creative.Version.Release.33.00\")"
        (:OlderVersion
          (bind _output_get_3 (Variables|UserOptions|Player|Inventory|SetbInfiniteWorldResources (Variables|Legacy|GetbInfiniteResources)))
          (bind _output_get_4 (Variables|UserOptions|Player|Inventory|SetbInfiniteBuildingResources _output_get_3))
          (Variables|UserOptions|Player|Inventory|SetbInfiniteGold _output_get_4)
          (bind _output_get_5 (Variables|UserOptions|Overrides|SetbInfiniteWorldResources_Override (Variables|Legacy|GetbInfiniteResources_Override)))
          (bind _output_get_6 (Variables|UserOptions|Overrides|SetbInfiniteBuildingResources_Override _output_get_5))
          (Variables|UserOptions|Overrides|SetbInfiniteGold_Override _output_get_6)
          (LocalVariables|SetbWasMigrated true))))
    (:then_2
      (CreativeVersion|CompareVersiontoTag (Variables|GetLast%20Saved%20Version) "(TagName=\"Creative.Version.Release.35.20\")"
        (:OlderVersion
          (Variables|UserOptions|Overrides|SetbDisplayEmptyAmmoSlots_Override true)
          (LocalVariables|SetbWasMigrated true))))
    (:then_3
      (CreativeVersion|CompareVersiontoTag (Variables|GetLast%20Saved%20Version) "(TagName=\"Creative.Version.IslandSettings.HealthShieldVariablePeriod\")"
        (:OlderVersion
          ;[1 1 1] Preserve prior experience tick period and amounts
          (Variables|UserOptions|Player|Health|SetHealthRechargePeriod 0.2)
          (Variables|UserOptions|Player|Shields|SetShieldRechargePeriod 0.2)
          (Variables|UserOptions|Player|Shields|SetOvershieldRechargePeriod 0.2)
          (Variables|UserOptions|Player|Health|SetHealthRechargeApplicationAmount (Variables|Legacy|GetHealthRechargeAmount))
          (Variables|UserOptions|Player|Shields|SetShieldRechargeApplicationAmount (Variables|Legacy|GetShieldRechargeAmount))
          (Variables|UserOptions|Player|Shields|SetOvershieldRechargeApplicationAmount (Variables|Legacy|GetOvershieldRechargeRate))
          (Variables|UserOptions|Overrides|SetOvershieldRechargeApplicationAmount_Override (or (Variables|Legacy|GetOvershieldRechargeRate_Override) (Variables|UserOptions|Overrides|GetOvershieldRechargeApplicationAmount_Override))) ; FORT-943099 the varguid was not set up correctly on first implementation, such that in UEFN the legacy data was saved directly onto the replacement variable, while FNC did not. OR the two variables to account for discrepancy.
          (LocalVariables|SetbWasMigrated true))))
    (:then_4
      (if (LocalVariables|GetbWasMigrated)
        (OptionData|RefreshPropertyOverrides (Variables|Default|GetToyOptionsComponent))))))

(fn InitializeMIDs)

(fn IsActorOnSameTeam
  (params (Actor :type Actor%20Object%20Reference))
  ;[1 1 1] Player team is actually set to Team + EFortTeam::HumanPVP so account for that in our team check
  (Utilities|IsValid Actor
    (:Is%20Valid
      (return (or (Creative|TeamOption|IsTeamType (Variables|UserOptions|General|GetTeamSetting) "All") (== (Team|GetActorTeam Actor) (Math|Byte|MakeLiteralByte (Creative|TeamOption|ConverttoTeam (Variables|UserOptions|General|GetTeamSetting)))))))
    (:Is%20Not%20Valid
      (return false))))

(fn IsAgentOnTeam
  (params (Controller :type Controller%20Object%20Reference))
  (return (CallFunction|IsActoronSameTeam Controller)))

(fn IsAnyItemBlacklisted
  (params (InPickup :type Fort%20Pickup%20Object%20Reference))
  (bind _index (Utilities|FlowControl|ForLoopwithBreak 0 (- (Pickup|GetItemEntryNum InPickup) 1))
    (:LoopBody
      (Utilities|IsBlacklisted (Class|FortItemFunctionLibrary|BreakFortItemEntry (Pickup|GetItemEntry (Variables|GetInPickup) _index))
        (:Blacklisted
          (LocalVariables|SetbWasBlacklisted true))))
    (:Completed
      _))
  (return (LocalVariables|GetbWasBlacklisted)))

(fn IsPlayerOnSameIsland
  (params (Actor :type Actor%20Object%20Reference))
  (bind _returnvalue (Class|CreativeBlueprintLibrary|IsActorRelevanttoActor self Actor))
  (return _returnvalue))

(fn IsTeamOutOfRespawns
  (bind _array_element (Utilities|Array|ForEachLoopwithBreak (Fortnite|GetAllFortPlayerControllers true))
    (:LoopBody
      (if (LocalVariables|GetbTeamIsOutOfRespawns)
        _
        (else
          _))
      ;[1 1 1] Body loop, determine if the player is on our team and if they have lives left
      (Utilities|IsValid _array_element
        (:Is%20Valid
          (bind _isinsameisland (CallFunction|IsPlayeronSameIsland _array_element))
          (if (and _isinsameisland (CallFunction|IsActoronSameTeam _array_element))
            (LocalVariables|SetTeamSize (Math|Byte|Byte+Byte (LocalVariables|GetTeamSize) 1))
            (Utilities|IsValid (Variables|/Script/Engine.Controller|GetPlayerState _array_element)
              (:Is%20Valid
                (bind _asfort_player_state (Utilities|Casting|CastToFortPlayerState (Variables|/Script/Engine.Controller|GetPlayerState _array_element)))
                (Utilities|IsValid (Class|FortPlayerState|FindFNEPlayerComponent _asfort_player_state)
                  (:Is%20Valid
                    (if (Class|FNEPlayerComponent|IsPlayerOutOfLives (Class|FortPlayerState|FindFNEPlayerComponent _asfort_player_state))
                      (Utilities|IsValid (Game|Player|GetPlayerPawn _array_element)
                        (:Is%20Valid
                          (if (FortPawn|IsDead (Game|Player|GetPlayerPawn _array_element))
                            _
                            (else
                              _))))
                      (else
                        _))
                    (LocalVariables|SetbTeamIsOutOfRespawns false)))))))))
    (:Completed
      _))
  (return (and (LocalVariables|GetbTeamIsOutOfRespawns) (Math|Byte|Byte>Byte (LocalVariables|GetTeamSize) 0))))

(fn MaybeGiveAmmoToInventoryOwner
  (params (InventoryOwner :type Fort%20Inventory%20Owner%20Interface%20Interface) (AmmoDefinition :type Fort%20World%20Item%20Definition%20Object%20Reference))
  (HelperFunctions|GiveAmmotoInventoryOwner InventoryOwner AmmoDefinition -1 (Variables|Default|GetExtra%20Ammo%20Amounts%20Map) (Variables|Default|GetDefaultExtraAmmoAmount) (Variables|UserOptions|Overrides|GetSpareWeaponAmmo_Override) (Variables|UserOptions|Player|Inventory|GetSpareWeaponAmmo)))

(fn MigrateDataToPickupList
  (bind _tospawnlist (Variables|Default|GetToSpawnList))
  (bind _returnvalue (Utilities|Array|Length _tospawnlist))
  (bind _returnvalue_2 (> _returnvalue 0))
  (bind _tospawnlist_2 (Variables|Default|GetToSpawnList))
  ;[1 1 1] This function is here to fill up PickupItemList Component on islands previous this device refactor on 12.30.
  ; After the refactor, PickupItemList is in charge of items but those were previously stored on ToSpawnList.
  ; If ToSpawnList contains items and PickupItemList doesn't, then, fill the list with already saved data
  (if (and _returnvalue_2 (== (ItemData|GetNumItems (Variables|Default|GetPickupItemList)) 0))
    (ItemData|AddItemsfromSpawnParams (Variables|Default|GetPickupItemList) _tospawnlist_2)
    (elif _returnvalue_2
      _))
  (Utilities|Array|Clear _tospawnlist_2))

(fn MigrateLegacyUserOptions
  (bind _sprintingspeedmultiplier (Class|GetClassDefaults "/CreativeCoreDevices/Device_TeamSettings_V2.Device_TeamSettings_V2_C"))
  (bind _sprintingjumpmultiplier (Class|GetClassDefaults "/CreativeCoreDevices/Device_TeamSettings_V2.Device_TeamSettings_V2_C"))
  (bind (_minimumledgeheight _minimumledgeheightwater _minimumledgeheight_override _minimumledgeheightwater_override) (Class|GetClassDefaults "/CreativeCoreDevices/Device_TeamSettings_V2.Device_TeamSettings_V2_C"))
  (bind (_mantlingheightparams _mantlingheightwaterparams) (Class|GetClassDefaults "/CreativeCoreDevices/ExperienceSettings/MantlingConversionConstants.MantlingConversionConstants_C"))
  (bind (_falldamagemultiplier _falldamagemultiplier_override) (Class|GetClassDefaults "/CreativeCoreDevices/Device_TeamSettings_V2.Device_TeamSettings_V2_C"))
  (bind (_maxitemslots _maxitemslots_override _maximumequipmentslots) (Class|GetClassDefaults "/CreativeCoreDevices/Device_TeamSettings_V2.Device_TeamSettings_V2_C"))
  (bind _maxitemslots_2 (Variables|Legacy|GetMaxItemSlots))
  (bind (_nameplatedisplaydistance _nameplatefocusangle _nameplatedisplaydistancemeters _nameplatedisplaydistance_override _nameplatefocusangle_override) (Class|GetClassDefaults "/CreativeCoreDevices/Device_TeamSettings_V2.Device_TeamSettings_V2_C"))
  (bind _nameplatedisplaydistance_2 (Variables|Legacy|GetNameplateDisplayDistance))
  (bind _returnvalue (> _nameplatedisplaydistance_2 0))
  (bind _nameplatefocusangle_2 (Variables|Legacy|GetNameplateFocusAngle))
  (bind (_spawnimmunitytime_legacy _spawnimmunitytime_legacy_override _spawnimmunitytime) (Class|GetClassDefaults "/CreativeCoreDevices/Device_TeamSettings_V2.Device_TeamSettings_V2_C"))
  (bind _spawnimmunitytime_legacy_2 (Variables|Legacy|GetSpawnImmunityTime_Legacy))
  (bind (_numcollectitemsforendcondition _collectitemcounttoend _numcollectitemsforendcondition_override) (Class|GetClassDefaults "/CreativeCoreDevices/Device_TeamSettings_V2.Device_TeamSettings_V2_C"))
  (bind _numcollectitemsforendcondition_2 (Variables|Legacy|GetNumCollectItemsForEndCondition))
  (bind (_startinghealthpercentage _startinghealthpercentage_override) (Class|GetClassDefaults "/CreativeCoreDevices/Device_TeamSettings_V2.Device_TeamSettings_V2_C"))
  (bind (_allowtacticalovershield _allowsprint _allowdoorbash _allowclambering _allowdoorbash_override _allowclambering_override _allowtacticalovershield_override _allowsprint_override _allowhealthshieldregenhealth _allowhealthshieldregenshield) (Class|GetClassDefaults "/CreativeCoreDevices/Device_TeamSettings_V2.Device_TeamSettings_V2_C"))
  (bind _bgrantammowithweapons (Class|GetClassDefaults "/CreativeCoreDevices/Device_TeamSettings_V2.Device_TeamSettings_V2_C"))
  (bind (_bapplyjumppenalty _bapplyjumppenalty_override) (Class|GetClassDefaults "/CreativeCoreDevices/Device_TeamSettings_V2.Device_TeamSettings_V2_C"))
  (bind (_bgliderredeploy _bgliderredeploy_override) (Class|GetClassDefaults "/CreativeCoreDevices/Device_TeamSettings_V2.Device_TeamSettings_V2_C"))
  (bind (_bflyenabled _bflyenabled_override) (Class|GetClassDefaults "/CreativeCoreDevices/Device_TeamSettings_V2.Device_TeamSettings_V2_C"))
  (|CallParentFunction)
  (Utilities|FlowControl|Sequence
    (:then_0
      ; MigrateSharedSettings
      (Utilities|FlowControl|Sequence
        (:then_0
          (bind _bmigrated (Creative|Migration|MigrateOptionTypeFloattoBoolean (Variables|Legacy|GetAllowDoorBash) _allowdoorbash (Variables|UserOptions|Player|Locomotion|GetbAllowShoulderBashing)))
          (bind _bmigrated_2 (Creative|Migration|MigrateOptionTypeFloattoBoolean (Variables|Legacy|GetAllowClambering) _allowclambering (Variables|UserOptions|Player|Locomotion|GetbAllowMantling)))
          (bind _bmigrated_3 (Creative|Migration|MigrateOptionTypeFloattoBoolean (Variables|Legacy|GetAllowTacticalOvershield) _allowtacticalovershield (Variables|UserOptions|Player|Shields|GetbAllowOvershield)))
          (bind _bmigrated_4 (Creative|Migration|MigrateOptionTypeFloattoBoolean (Variables|Legacy|GetAllowSprint) _allowsprint (Variables|UserOptions|Player|Locomotion|GetbAllowSprinting)))
          (bind _bmigrated_5 (Creative|Migration|MigrateOptionTypeFloattoBoolean (Variables|Legacy|GetAllowHealthShieldRegenHealth) _allowhealthshieldregenhealth (Variables|UserOptions|Player|Health|GetAllowHealthRecharge)))
          (bind _bmigrated_6 (Creative|Migration|MigrateOptionTypeFloattoBoolean (Variables|Legacy|GetAllowHealthShieldRegenShield) _allowhealthshieldregenshield (Variables|UserOptions|Player|Shields|GetAllowShieldRecharge))))
        (:then_1
          (bind _bmigrated_7 (Creative|Migration|MigrateOptionTypeBooleantoBoolean (Variables|Legacy|GetAllowDoorBash_Override) _allowdoorbash_override (Variables|UserOptions|Overrides|GetbAllowShoulderBashing_Override)))
          (bind _bmigrated_8 (Creative|Migration|MigrateOptionTypeBooleantoBoolean (Variables|Legacy|GetAllowClambering_Override) _allowclambering_override (Variables|UserOptions|Overrides|GetbAllowMantling_Override)))
          (bind _bmigrated_9 (Creative|Migration|MigrateOptionTypeBooleantoBoolean (Variables|Legacy|GetAllowTacticalOvershield_Override) _allowtacticalovershield_override (Variables|UserOptions|Overrides|GetbAllowOvershield_Override)))
          (bind _bmigrated_10 (Creative|Migration|MigrateOptionTypeBooleantoBoolean (Variables|Legacy|GetAllowSprint_Override) _allowsprint_override (Variables|UserOptions|Overrides|GetbAllowSprinting_Override)))
          (Variables|SetbWasMigrated (Math|Boolean|ORBoolean (Variables|GetbWasMigrated) (Math|Boolean|ORBoolean _bmigrated_6 _bmigrated_5 _bmigrated_4 _bmigrated_3 _bmigrated_2 _bmigrated) (Math|Boolean|ORBoolean _bmigrated_7 _bmigrated_8 _bmigrated_9 _bmigrated_10))))
        (:then_2
          _))
      ; MigrateMantlingOptions
      (Utilities|FlowControl|Sequence
        (:then_0
          (bind (_bmigrated_11 _newoptionvalue) (Class|MantlingConversionHelpers|MigrateOptionTypeFloattoMantlingHeightPreset (Variables|Legacy|GetMinimumLedgeHeight) _minimumledgeheight _mantlingheightparams))
          (if _bmigrated_11
            (Variables|UserOptions|Player|Locomotion|SetMantlingMinimumHeight (Utilities|Enum|BytetoEnumEnum_Common_MantlingLedgeHeight _newoptionvalue))
            (Variables|SetbWasMigrated true)))
        (:then_1
          (bind (_bmigrated_12 _newoptionvalue_2) (Class|MantlingConversionHelpers|MigrateOptionTypeFloattoMantlingHeightPreset (Variables|Legacy|GetMinimumLedgeHeightWater) _minimumledgeheightwater _mantlingheightwaterparams))
          (if _bmigrated_12
            (Variables|UserOptions|Player|Locomotion|SetMantlingMinimumHeightInWater (Utilities|Enum|BytetoEnumEnum_Common_MantlingLedgeHeight _newoptionvalue_2))
            (Variables|SetbWasMigrated true)))
        (:then_2
          (bind _bmigrated_13 (Creative|Migration|MigrateOptionTypeBooleantoBoolean (Variables|Legacy|GetMinimumLedgeHeight_Override) _minimumledgeheight_override (Variables|UserOptions|Overrides|GetMantlingMinimumHeight_Override)))
          (bind _bmigrated_14 (Creative|Migration|MigrateOptionTypeBooleantoBoolean (Variables|Legacy|GetMinimumLedgeHeightWater_Override) _minimumledgeheightwater_override (Variables|UserOptions|Overrides|GetMantlingMinimumHeightInWater_Override)))
          (Variables|SetbWasMigrated (Math|Boolean|ORBoolean _bmigrated_14 _bmigrated_13 (Variables|GetbWasMigrated))))
        (:then_3
          _))
      ; MigrateStartingHealth
      (Utilities|FlowControl|Sequence
        (:then_0
          (if (<= (Variables|UserOptions|Player|Health|GetStartingHealthPercentage) 0)
            (Variables|UserOptions|Player|Health|SetbInvincibility true)
            (Variables|UserOptions|Player|Health|SetStartingHealthPercentage _startinghealthpercentage)
            (Variables|SetbWasMigrated true)
            (Creative|Migration|MigrateOptionTypeBooleantoBoolean (Variables|UserOptions|Overrides|GetStartingHealthPercentage_Override) _startinghealthpercentage_override (Variables|UserOptions|Overrides|GetbInvincibility_Override))))
        (:then_1
          _))
      ; MigrateFallDamage
      (Utilities|FlowControl|Sequence
        (:then_0
          (bind _bmigrated_15 (Creative|Migration|MigrateOptionTypeFloattoBoolean (Variables|Legacy|GetFallDamageMultiplier) _falldamagemultiplier (Variables|UserOptions|Player|Locomotion|GetbFallDamage)))
          (bind _bmigrated_16 (Creative|Migration|MigrateOptionTypeBooleantoBoolean (Variables|Legacy|GetFallDamageMultiplier_Override) _falldamagemultiplier_override (Variables|UserOptions|Overrides|GetbFallDamage_Override)))
          (Variables|SetbWasMigrated (Math|Boolean|ORBoolean _bmigrated_16 _bmigrated_15 (Variables|GetbWasMigrated))))
        (:then_1
          _))
      ; MigrateMaxItemSlots
      (Utilities|FlowControl|Sequence
        (:then_0
          (if (!= _maxitemslots _maxitemslots_2)
            (Variables|UserOptions|Player|Equipment|SetMaximumEquipmentSlots (select (>= _maxitemslots_2 0) (Math|Float|Truncate _maxitemslots_2) _maximumequipmentslots))
            (Variables|Legacy|SetMaxItemSlots _maxitemslots)
            (Variables|SetbWasMigrated true)))
        (:then_1
          (bind _bmigrated_17 (Creative|Migration|MigrateOptionTypeBooleantoBoolean (Variables|Legacy|GetMaxItemSlots_Override) _maxitemslots_override (Variables|UserOptions|Overrides|GetMaximumEquipmentSlots_Override)))
          (if _bmigrated_17
            (Variables|SetbWasMigrated true)))
        (:then_2
          _))
      ; MigrateSprintingSpeedMultipliers
      (Utilities|FlowControl|Sequence
        (:then_0
          (if (not (Math|Float|NearlyEqual(Float) (Variables|Legacy|GetSprintingSpeedMultiplier) _sprintingspeedmultiplier 0.001))
            (Variables|UserOptions|Player|Locomotion|SetMaxSprintingSpeedMultiplier (+ (* (- (Variables|Legacy|GetSprintingSpeedMultiplier) 1.0) 1e+01) 1.0))
            (Variables|Legacy|SetSprintingSpeedMultiplier _sprintingspeedmultiplier)
            (Variables|SetbWasMigrated true)))
        (:then_1
          (if (not (Math|Float|NearlyEqual(Float) (Variables|Legacy|GetSprintingJumpMultiplier) _sprintingjumpmultiplier 0.001))
            (Variables|UserOptions|Player|Locomotion|SetMaxSprintingJumpMultiplier (+ (* (- (Variables|Legacy|GetSprintingJumpMultiplier) 1.0) 1e+01) 1.0))
            (Variables|Legacy|SetSprintingJumpMultiplier _sprintingjumpmultiplier)
            (Variables|SetbWasMigrated true)))
        (:then_2
          (bind _bmigrated_18 (Creative|Migration|MigrateOptionTypeBooleantoBoolean (Variables|Legacy|GetSprintingJumpMultiplier_Override) (Class|GetClassDefaults "/CreativeCoreDevices/Device_TeamSettings_V2.Device_TeamSettings_V2_C") (Variables|UserOptions|Overrides|GetMaxSprintingJumpMultiplier_Override)))
          (Variables|SetbWasMigrated (or _bmigrated_18 (Variables|GetbWasMigrated))))
        (:then_3
          (bind _bmigrated_19 (Creative|Migration|MigrateOptionTypeBooleantoBoolean (Variables|Legacy|GetSprintingSpeedMultiplier_Override) (Class|GetClassDefaults "/CreativeCoreDevices/Device_TeamSettings_V2.Device_TeamSettings_V2_C") (Variables|UserOptions|Overrides|GetMaxSprintingSpeedMultiplier_Override)))
          (Variables|SetbWasMigrated (or _bmigrated_19 (Variables|GetbWasMigrated))))
        (:then_4
          _)))
    (:then_1
      ; MigrateNameplateSettings
      (Utilities|FlowControl|Sequence
        (:then_0
          (bind _bmigrated_20 (Creative|Migration|MigrateOptionTypeBooleantoBoolean (Variables|Legacy|GetNameplateDisplayDistance_Override) _nameplatedisplaydistance_override (Variables|UserOptions|Overrides|GetNameplateDisplayDistanceMeters_Override)))
          (if _bmigrated_20
            (Variables|UserOptions|Overrides|SetbLimitNameplateMaxDistance_Override (Variables|UserOptions|Overrides|GetNameplateDisplayDistanceMeters_Override))
            (Variables|SetbWasMigrated true)))
        (:then_1
          (if (!= _nameplatedisplaydistance _nameplatedisplaydistance_2)
            (Variables|UserOptions|InGameInterface|Nameplate|SetbLimitNameplateMaxDistance _returnvalue)
            (Variables|UserOptions|InGameInterface|Nameplate|SetNameplateDisplayDistanceMeters (select _returnvalue (/ _nameplatedisplaydistance_2 1e+02) _nameplatedisplaydistancemeters))
            (Variables|Legacy|SetNameplateDisplayDistance _nameplatedisplaydistance)
            (Variables|SetbWasMigrated true)))
        (:then_2
          (if (!= _nameplatefocusangle _nameplatefocusangle_2)
            (Variables|UserOptions|InGameInterface|Nameplate|SetNameplateFocusAngleDegrees (Math|Conversions|ToFloat(Integer) (Math|Float|Round (Math|Trig|RadiansToDegrees _nameplatefocusangle_2))))
            (Variables|Legacy|SetNameplateFocusAngle _nameplatefocusangle)
            (Variables|SetbWasMigrated true)))
        (:then_3
          (bind _bmigrated_21 (Creative|Migration|MigrateOptionTypeBooleantoBoolean (Variables|Legacy|GetNameplateFocusAngle_Override) _nameplatefocusangle_override (Variables|UserOptions|Overrides|GetNameplateFocusAngleDegrees_Override)))
          (if _bmigrated_21
            (Variables|SetbWasMigrated true)))
        (:then_4
          (if (< (Variables|UserOptions|InGameInterface|Nameplate|GetNameplateFocusTime) 0)
            (Variables|UserOptions|InGameInterface|Nameplate|SetNameplateFocusTime 0.0)
            (Variables|SetbWasMigrated true)))
        (:then_5
          _))
      ; MigrateSpawnImmunityTime
      (Utilities|FlowControl|Sequence
        (:then_0
          (if (!= _spawnimmunitytime_legacy _spawnimmunitytime_legacy_2)
            (bind _output_get (Variables|UserOptions|Mode|Spawning|SetbOverrideSpawnImmunityTime (>= _spawnimmunitytime_legacy_2 0)))
            (Variables|UserOptions|Mode|Spawning|SetSpawnImmunityTime (select _output_get _spawnimmunitytime_legacy_2 _spawnimmunitytime))
            (Variables|Legacy|SetSpawnImmunityTime_Legacy _spawnimmunitytime_legacy)
            (Variables|SetbWasMigrated true)))
        (:then_1
          (bind _bmigrated_22 (Creative|Migration|MigrateOptionTypeBooleantoBoolean (Variables|Legacy|GetSpawnImmunityTime_Legacy_Override) _spawnimmunitytime_legacy_override (Variables|UserOptions|Overrides|GetSpawnImmunityTime_Override)))
          (if _bmigrated_22
            (Variables|UserOptions|Overrides|SetbOverrideSpawnImmunityTime_Override (Variables|UserOptions|Overrides|GetSpawnImmunityTime_Override))
            (Variables|SetbWasMigrated true)))
        (:then_2
          _))
      ; MigrateCollectItemsToEnd
      (Utilities|FlowControl|Sequence
        (:then_0
          (if (!= _numcollectitemsforendcondition _numcollectitemsforendcondition_2)
            (bind _output_get_2 (Variables|UserOptions|Round|EndCondition|SetCollectItemsToEnd (select (!= _numcollectitemsforendcondition_2 0) (select (> _numcollectitemsforendcondition_2 0) "NewEnumerator1" "NewEnumerator0") "NewEnumerator2")))
            (Variables|UserOptions|Round|EndCondition|SetCollectItemCountToEnd (Utilities|Select :Index _output_get_2 :NewEnumerator2 _collectitemcounttoend :NewEnumerator0 _collectitemcounttoend :NewEnumerator1 _numcollectitemsforendcondition_2))
            (Variables|Legacy|SetNumCollectItemsForEndCondition _numcollectitemsforendcondition)
            (Variables|SetbWasMigrated true)))
        (:then_1
          (bind _bmigrated_23 (Creative|Migration|MigrateOptionTypeBooleantoBoolean (Variables|Legacy|GetNumCollectItemsForEndCondition_Override) _numcollectitemsforendcondition_override (Variables|UserOptions|Overrides|GetCollectItemsToEnd_Override)))
          (if _bmigrated_23
            (Variables|SetbWasMigrated true)))
        (:then_2
          _))
      ; Migrate Grant Ammo With Weapons
      (if (!= _bgrantammowithweapons (Variables|Legacy|GetbGrantAmmoWithWeapons))
        (Variables|UserOptions|Overrides|SetSpareWeaponAmmo_Override (not (Variables|Legacy|GetbGrantAmmoWithWeapons)))
        (Variables|UserOptions|Player|Inventory|SetSpareWeaponAmmo 0)
        (Variables|Legacy|SetbGrantAmmoWithWeapons _bgrantammowithweapons)
        (Variables|SetbWasMigrated true)
        (else
          _))
      ; MigrateLocomotionSettings
      (bind _bmigrated_24 (Creative|Migration|MigrateOptionTypeBooleantoBoolean (Variables|Legacy|GetbApplyJumpPenalty) _bapplyjumppenalty (Variables|UserOptions|Player|Locomotion|GetbApplyJumpingPenalty)))
      (bind _bmigrated_25 (Creative|Migration|MigrateOptionTypeBooleantoBoolean (Variables|Legacy|GetbApplyJumpPenalty_Override) _bapplyjumppenalty_override (Variables|UserOptions|Overrides|GetbApplyJumpingPenalty_Override)))
      (bind _bmigrated_26 (Creative|Migration|MigrateOptionTypeBooleantoBoolean (Variables|Legacy|GetbGliderRedeploy) _bgliderredeploy (Variables|UserOptions|Player|Locomotion|GetbGliderRedeployable)))
      (bind _bmigrated_27 (Creative|Migration|MigrateOptionTypeBooleantoBoolean (Variables|Legacy|GetbGliderRedeploy_Override) _bgliderredeploy_override (Variables|UserOptions|Overrides|GetbGliderRedeployable_Override)))
      (bind _bmigrated_28 (Creative|Migration|MigrateOptionTypeBooleantoBoolean (Variables|Legacy|GetbFlyEnabled) _bflyenabled (Variables|UserOptions|Player|Locomotion|GetbFlyingEnabled)))
      (bind _bmigrated_29 (Creative|Migration|MigrateOptionTypeBooleantoBoolean (Variables|Legacy|GetbFlyEnabled_Override) _bflyenabled_override (Variables|UserOptions|Overrides|GetbFlyingEnabled_Override)))
      (Variables|SetbWasMigrated (Math|Boolean|ORBoolean _bmigrated_29 _bmigrated_28 _bmigrated_27 _bmigrated_26 _bmigrated_25 _bmigrated_24)))
    (:then_2
      (if (LocalVariables|GetbWasMigrated)
        (OptionData|RefreshPropertyOverrides (Variables|Default|GetToyOptionsComponent))))))

(fn RemoveAllPlayerBinds
  (params (Minigame :type Fort%20Minigame%20Object%20Reference))
  ;[1 0.298 0.1277] Unbinds bound events on all players
  (Utilities|FlowControl|Sequence
    (:then_0
      (Utilities|IsValid Minigame
        (:Is%20Valid
          (for _array_element (Minigame|GetParticipatingPlayers Minigame)
            (Utilities|IsValid _array_element
              (:Is%20Valid
                (Utilities|IsValid (Fortnite|GetControllerfromActor _array_element)
                  (:Is%20Valid
                    (Pawn|UnbindEventfromOnPossessedPawnChanged (Fortnite|GetControllerfromActor _array_element) (EventDispatchers|CreateEvent :!event OnPossessedPawnChanged))))))))))
    (:then_1
      (bind _outfortplayerpawns (Fortnite|GetFortPlayerPawns))
      (for _array_element_2 _outfortplayerpawns
        (Utilities|IsValid _array_element_2
          (:Is%20Valid
            (EventDispatchers|UnbindEventfromOnDied _array_element_2 (EventDispatchers|CreateEvent :!event OnPlayerPawnDied))))))))

(fn RemoveTeamSettingsFromMinigame
  (params (Minigame :type Fort%20Minigame%20Object%20Reference))
  (Utilities|FlowControl|Sequence
    (:then_0
      (bind _output_get (LocalVariables|SetTeamSettingsMutator (TeamSettings|GetTeamSettingsMutator (Variables|GetMinigame))))
      (Utilities|IsValid _output_get
        (:Is%20Valid
          (bind _returnvalue (TeamSettings|RemoveDataSource _output_get (Variables|Default|GetToyOptionsComponent)))
          (if _returnvalue
            (Minigame|GameRules|RefreshGameRulesPreview (Variables|GetMinigame))))))
    (:then_1
      (bind _output_get_2 (LocalVariables|SetWeaponSettingsMutator (TeamSettings|GetWeaponSettingsMutator (Variables|GetMinigame))))
      (Utilities|IsValid _output_get_2
        (:Is%20Valid
          (Class|FortAthenaMutatorWeaponSettings|RemoveDataSource _output_get_2 (Variables|Default|GetToyOptionsComponent)))))
    (:then_2
      (Class|FortAthenaMutatorUtilities|RemoveDataRegistryOptionsforContextObject))))

(fn RemoveTemporaryMemberFromTeam_Verse
  (params (Agent :type Controller%20Object%20Reference))
  (Class|FortPlayspaceComponentTemporaryTeams|HavePlayerLeaveTemporaryTeam Agent self "FromVerse"))

(fn RemoveTemporaryMembersFromTeam_Verse
  (params (Logic_Unused :type Boolean))
  (bind _teamsetting (Variables|UserOptions|General|GetTeamSetting))
  (switch Utilities|FlowControl|Switch|SwitchonECreativeTeamType (Utilities|Struct|BreakCreativeTeamOption _teamsetting)
    (:TeamIndex
      (Class|FortPlayspaceComponentTemporaryTeams|RemoveAllTemporaryTeamMembersforTeam self (Creative|TeamOption|ConverttoTeam _teamsetting) "FromVerse"))
    (:All
      (Class|FortPlayspaceComponentTemporaryTeams|RemoveAllTemporaryTeams self "FromVerse"))))

(fn ResaveActor
  (CallFunction|UpdateTeamAssignmentDisplay)
  (CallFunction|Resave))

(fn RespawnPlayerAtPlayerSpawner_Verse
  (params (Agent :type Controller%20Object%20Reference))
  (bind _asplayer_controller (Utilities|Casting|CastToPlayerController Agent))
  (CallFunction|RespawnatPlayerSpawnerEvent _asplayer_controller))

(fn RespawnPlayerFromPlayerSpawnerGroup_AddSpawner_Verse
  (params (Spawner :type BP%20Creative%20Player%20Spawner%20Prop%20Object%20Reference))
  ; We add spawners to consider one at a time and then execute on them in the DoWork function.  It'd be preferable to pass the array, but since BP always marks arrays as
  ; CPF_OutParm, we can't safely pass it directly from verse
  (bind _asfort_player_start_creative (Utilities|Casting|CastToFortPlayerStartCreative (Variables|/Script/Engine.ChildActorComponent|GetChildActor (Variables|/CRD_PlayerSpawn/BP_Creative_Player_Spawner_Prop.BP_Creative_Player_Spawner_Prop_C|GetPlayerStartCreative Spawner))))
  (Utilities|Array|Add (Variables|Default|GetVerseSpawnGroupSearchGroup) _asfort_player_start_creative))

(fn RespawnPlayerFromPlayerSpawnerGroup_DoWork_Verse
  (params (Player :type Controller%20Object%20Reference))
  (bind _returnvalue (Class|FortMinigamePlayerSpawnComponent|BPChoosePlayerStart (Variables|/Script/FortniteGame.FortMinigame|GetSpawnComponent (Minigame|GetMinigame self :!class /Script/FortniteGame.MinigameBlueprintLibrary)) Player (Variables|Default|GetVerseSpawnGroupSearchGroup) true true))
  (Utilities|Array|Clear (Variables|Default|GetVerseSpawnGroupSearchGroup))
  (bind _asfort_player_start_creative (Utilities|Casting|CastToFortPlayerStartCreative _returnvalue))
  (bind _asbp_creative_player_spawner_prop (Utilities|Casting|CastToBP_Creative_Player_Spawner_Prop (Actor|GetParentActor _asfort_player_start_creative)))
  (Class|BPCreativePlayerSpawnerProp|SpawnPlayerEvent _asbp_creative_player_spawner_prop (Variables|GetPlayer)))

(fn SetAllowTemporaryTeamMemberInvitations_Override_Verse
  (params (InOverriden :type Boolean))
  (Variables|Verse|SetAllowTemporaryTeamMemberInvitations_VerseOverride_IsSet InOverriden)
  (return))

(fn SetAllowTemporaryTeamMemberInvitations_Verse
  (params (InAllowed :type Boolean))
  (Variables|Verse|SetAllowTemporaryTeamMemberInvitations_VerseOverride InAllowed)
  (return))

(fn SetupItemPreview
  (Utilities|FlowControl|Sequence
    (:then_0
      (if (Editor|InEditor)
        (Class|CreativeItemPreviewComponent|EditorDisplayPreview (Variables|Default|GetCreative_ItemPreview_Component) (Variables|Default|GetPreviewStaticMesh))))
    (:then_1
      (Networking|CanExecuteCosmeticEvents
        (:True
          (Initialization|InitializeComponent (Variables|Default|GetCreative_ItemPreview_Component) (Variables|Default|GetRotationSceneComponent) (Variables|Default|GetPreviewStaticMesh) (Variables|Default|GetPreviewSkeletalMesh) 0 (Variables|Default|GetPickupItemList) (Variables|Default|GetItemVariantNotifier) :!class /Game/Creative/Devices/Common/Components/Creative_ItemPreview_Component.Creative_ItemPreview_Component_C))))))

(fn TryGiveAllItemsToPlayer
  (params (PlayerToGiveItemsTo :type Fort%20Inventory%20Owner%20Interface%20Interface))
  (bind _returnvalue (Math|Transform|MakeTransform "0, 0, 0"))
  (bind _pickupitemlist (Variables|Default|GetPickupItemList))
  (bind _returnvalue_2 (ItemData|MakeSpawnParamsList _pickupitemlist _returnvalue))
  (bind _playertogiveitemsto (Variables|GetPlayerToGiveItemsTo))
  (bind _playertogiveitemsto_2 (Variables|GetPlayerToGiveItemsTo))
  (bind _asfort_player_controller_athena (Utilities|Casting|CastToFortPlayerControllerAthena PlayerToGiveItemsTo)
    (:then
      (LocalVariables|SetPlayerController _asfort_player_controller_athena))
    (:CastFailed
      _))
  (Utilities|FlowControl|Sequence
    (:then_0
      (switch Utilities|FlowControl|Switch|SwitchonE_TeamSettings_GrantBehavior (Variables|UserOptions|Player|Inventory|GetOnGrantBehavior)
        (:NewEnumerator0
          (Inventory|ClearDroppableItems _playertogiveitemsto))
        (:NewEnumerator1
          (Inventory|RemoveNonResourceDroppableItems _playertogiveitemsto)))
      (bind _playercontroller (LocalVariables|GetPlayerController)
        (:then
          (Variables|/Script/FortniteGame.FortPlayerControllerAthena|SetbShouldForceDeleteDroppedItems true _playercontroller))
        (:else)))
    (:then_1
      (bind _output_get (LocalVariables|SetEquipItemIndexOnGrant (- (Math|Conversions|ToInteger(Byte) (Variables|UserOptions|Player|Inventory|GetEquipGrantedItem)) 1)))
      (for (_ _array_index) _returnvalue_2
        ;[1 0.1904 0] Is this the element we want to equip after gifting?
        (bind _structout (Default|TryOverrideInitialWeaponAmmunition (Variables|Default|GetPickupItemList) (Utilities|Array|Get(aref) _returnvalue_2 _array_index) (Variables|UserOptions|Overrides|GetInitialWeaponAmmo_Override) (Variables|UserOptions|Player|Inventory|GetInitialWeaponAmmo)))
        (if (Math|Boolean|ANDBoolean (Utilities|IsValid (Utilities|ResolveSoftReference (Utilities|Struct|BreakMinigameSpawnerSpawnParams _structout)) :!class /Script/Engine.KismetSystemLibrary) (not (Validation|IsBlockListedItemDefinition (Utilities|ResolveSoftReference (Utilities|Struct|BreakMinigameSpawnerSpawnParams _structout)))) (Validation|IsItemDefinitionAllowed (Utilities|ResolveSoftReference (Utilities|Struct|BreakMinigameSpawnerSpawnParams _structout))))
          (Inventory|GiveItemtoInventoryOwner _playertogiveitemsto_2 (Utilities|ResolveSoftReference (Utilities|Struct|BreakMinigameSpawnerSpawnParams _structout)) (.ItemVariantGuid (Utilities|Struct|BreakMinigameSpawnerSpawnParams _structout)) (.PickupQuantity (Utilities|Struct|BreakMinigameSpawnerSpawnParams _structout)) true -1 -1 true (.WeaponAmmoOverride (Utilities|Struct|BreakMinigameSpawnerSpawnParams _structout)))
          (CallFunction|MaybeGiveAmmotoInventoryOwner _playertogiveitemsto_2 (Utilities|ResolveSoftReference (Utilities|Struct|BreakMinigameSpawnerSpawnParams _structout)))
          (if (== (Math|Integer|Min(Integer) (Utilities|Array|LastIndex _returnvalue_2) _output_get) _array_index)
            (bind _playercontroller_2 (LocalVariables|GetPlayerController)
              (:then
                (Inventory|ExecuteInventoryItemDefinition :self _playercontroller_2 :ItemDefinition (Utilities|ResolveSoftReference (Utilities|Struct|BreakMinigameSpawnerSpawnParams _structout)) :Delay 0.0 :bForceExecute true :bActivateSlotAfterSettingFocused true))
              (:else)))))
      (bind _playercontroller_3 (LocalVariables|GetPlayerController)
        (:then
          (Variables|/Script/FortniteGame.FortPlayerControllerAthena|SetbShouldForceDeleteDroppedItems false _playercontroller_3))
        (:else)))))

(fn UpdateColors
  (bind _teamsetting (Variables|UserOptions|General|GetTeamSetting))
  (bind _teamsetting_2 (Variables|UserOptions|General|GetTeamSetting))
  (Networking|CanExecuteCosmeticEvents
    (:True
      (Utilities|FlowControl|Sequence
        (:then_0
          (bind _fortminigameprogress (Variables|Default|GetFortMinigameProgress)
            (:then
              (bind _currentminigame (Variables|/Script/FortniteGame.FortMinigameLogicComponent|GetCurrentMinigame _fortminigameprogress)
                (:then
                  (if (Creative|TeamOption|IsTeamType _teamsetting "All")
                    (LocalVariables|Setcolor%20Index -1)
                    (else
                      (Utilities|IsValid (Minigame|GetMinigameTeamComponent _currentminigame)
                        (:Is%20Valid
                          (LocalVariables|Setcolor%20Index (Minigame|GameRules|GetTeamColorIndex (Minigame|GetMinigameTeamComponent _currentminigame) (Math|Byte|MakeLiteralByte (Creative|TeamOption|ConverttoTeam _teamsetting)))))
                        (:Is%20Not%20Valid
                          (LocalVariables|Setcolor%20Index -1))))))
                (:else
                  ;[0.8827 0.1953 1 0.25] Editor Logic, ignore minigame, use team value directly.
                  (if (Editor|InEditor)
                    (if (Creative|TeamOption|IsTeamType _teamsetting_2 "All")
                      (LocalVariables|Setcolor%20Index -1)
                      (else
                        _)))
                  (LocalVariables|Setcolor%20Index (Math|Integer|Max(Integer) (Math|Integer|%25%28Integer%29 (Creative|TeamOption|GetLegacyTeamIndex _teamsetting_2 false) 16))))))
            (:else)))
        (:then_1
          (if (Variables|UserOptions|Overrides|GetTeamColor_Override)
            (LocalVariables|Setcolor%20Index (Math|Conversions|ToInteger(Byte) (Variables|UserOptions|General|GetTeamColor)))))
        (:then_2
          (bind _result (ArrayMacros|SafeArrayGet (Class|GetClassDefaults "/Game/Creative/Sets/MilitaryBase/Props/CreativeTeamColors.CreativeTeamColors_C") (select (== (LocalVariables|Getcolor%20Index) -1) (select (Creative|TeamOption|IsTeamType (Variables|UserOptions|General|GetTeamSetting) "All") 0 (Creative|TeamOption|GetLegacyTeamIndex (Variables|UserOptions|General|GetTeamSetting) false)) (LocalVariables|Getcolor%20Index)))
            (:%20%20
              (DeviceVisuals|ApplyColors _result))))))))

(fn ConstructionScript
  (|CallParentFunction))
