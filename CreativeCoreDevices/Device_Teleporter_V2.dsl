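; Fortnite Creative Teleporter V2 device that sends players, NPC pawns and physics props between teleporters.
; It registers by group with the minigame's TeleporterManager and teleports on capsule overlap or incoming events, gated by enabled state, restrictions, blocking tags and recent-teleport checks.
; Helper functions compute exit location, rotation and momentum, while an effect radius tracks overlapping actors.
; Rift effects animate through custom primitive data, and version upgrades migrate older prop-momentum options.
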(event OnAnyPropertyChanged(ToyOptionsComponent)
  (bind _knob_teleportergroup (Variables|UserOptions|Basic|GetKnob_TeleporterGroup))
  ;[0 1 0.3723 0.1] Device Initialization
  (Utilities|FlowControl|Sequence
    (:then_0
      (bind _teleportermanagercomponent (CallFunction|GetTeleporterManager))
      (Utilities|IsValid _teleportermanagercomponent
        (:Is%20Valid
          (Teleporter|UnregisterTeleporterfromGroup _teleportermanagercomponent self (Variables|Default|GetLastChannel))
          (Teleporter|RegisterTeleporterwithGroup _teleportermanagercomponent self _knob_teleportergroup)
          (Variables|Default|SetLastChannel _knob_teleportergroup))))
    (:then_1
      (Class|CreativeEnabledComponent|Initialize (Variables|Default|GetEnabledComponent) (Variables|Default|GetFortMinigameLogic) (Variables|UserOptions|GetEnabledDuringPhase))
      (CallFunction|UpdateRiftVisibility)
      (CallFunction|Resave)
      (Initialization|Initialize (Variables|Default|GetCreative_UseRestriction_Component) (Variables|Default|GetFortMinigameLogic) (Variables|Default|GetEnabledComponent) 0 (Variables|UserOptions|GetSelected%20Team) (Variables|UserOptions|GetInvert%20Team%20Selection) (Variables|UserOptions|GetSelected%20Class) (Variables|UserOptions|GetInvert%20Class%20Selection) false)
      (CallFunction|UpdateEffectRadiusScale)
      (Default|ResetLink)
      (CallFunction|InitializeFX)
      (Default|UpdatePhysicsObjectCollisionSettings))))

(event OnPropertiesLoaded(ToyOptionsComponent)
  (bind _knob_teleportergroup (Variables|UserOptions|Basic|GetKnob_TeleporterGroup))
  ;[0 1 0.3723 0.1] Device Initialization
  (Utilities|FlowControl|Sequence
    (:then_0
      (bind _teleportermanagercomponent (CallFunction|GetTeleporterManager))
      (Utilities|IsValid _teleportermanagercomponent
        (:Is%20Valid
          (Teleporter|UnregisterTeleporterfromGroup _teleportermanagercomponent self (Variables|Default|GetLastChannel))
          (Teleporter|RegisterTeleporterwithGroup _teleportermanagercomponent self _knob_teleportergroup)
          (Variables|Default|SetLastChannel _knob_teleportergroup))))
    (:then_1
      (Class|CreativeEnabledComponent|Initialize (Variables|Default|GetEnabledComponent) (Variables|Default|GetFortMinigameLogic) (Variables|UserOptions|GetEnabledDuringPhase))
      (CallFunction|UpdateRiftVisibility)
      (CallFunction|Resave)
      (Initialization|Initialize (Variables|Default|GetCreative_UseRestriction_Component) (Variables|Default|GetFortMinigameLogic) (Variables|Default|GetEnabledComponent) 0 (Variables|UserOptions|GetSelected%20Team) (Variables|UserOptions|GetInvert%20Team%20Selection) (Variables|UserOptions|GetSelected%20Class) (Variables|UserOptions|GetInvert%20Class%20Selection) false)
      (CallFunction|UpdateEffectRadiusScale)
      (Default|ResetLink)
      (CallFunction|InitializeFX)
      (Default|UpdatePhysicsObjectCollisionSettings))))

(event OnAddedtoMinigame(FortMinigameLogic)
  (params (Minigame :type Fort%20Minigame%20Object%20Reference))
  (bind _fortminigamelogic (Variables|Default|GetFortMinigameLogic))
  (bind _currentminigame (Variables|/Script/FortniteGame.FortMinigameLogicComponent|GetCurrentMinigame _fortminigamelogic))
  (bind _knob_teleportergroup (Variables|UserOptions|Basic|GetKnob_TeleporterGroup))
  ;[0 1 0.1023 0.1] Register ourself with the current minigame
  (Utilities|FlowControl|Sequence
    (:then_0
      (bind _teleportermanagercomponent (CallFunction|GetTeleporterManager))
      (Utilities|IsValid _teleportermanagercomponent
        (:Is%20Valid
          (Teleporter|RegisterTeleporterwithGroup _teleportermanagercomponent self _knob_teleportergroup)
          (Variables|Default|SetLastChannel _knob_teleportergroup))))
    (:then_1
      (Default|ResetLink)
      (CallFunction|UpdateRiftVisibility)
      (Minigame|BindEventtoOnPlayModeChangedDelegate _currentminigame (EventDispatchers|CreateEvent :!event Custom|OnPlayModeChanged))
      (CallFunction|OnPlayModeChanged _currentminigame (Minigame|IsinPlayMode _currentminigame :!class /Script/FortniteGame.FortMinigame)))))

(event EventOnWorldReady
  ;[0 1 0.1023 0.1] Register ourself with the current minigame
  ;;[0 1 0.4573 0] OnWorldReady setup visibility component so it's aware of our Effect Radius Mesh
  (|CallParentFunction)
  (Visibility|SetVisibilityComponents (Variables|Default|GetCreative_VisibleInGame_Component) (Utilities|Array|MakeArray (Variables|Default|GetEffectRadius)) true))

(event EditorPreview|EventSetupForEditorPreview
  ;[0 1 0.1023 0.1] Register ourself with the current minigame
  ;;[0 1 0.1757 0.25] If we're in editor, run the same logic as if we're not in play mode.
  (|CallParentFunction)
  (CallFunction|UpdateEffectRadiusMesh true)
  ;;[0 1 0.4573 0] OnWorldReady setup visibility component so it's aware of our Effect Radius Mesh
  (Visibility|SetVisibilityComponents (Variables|Default|GetCreative_VisibleInGame_Component) (Utilities|Array|MakeArray (Variables|Default|GetEffectRadius)) true))

(event MinigameStartedDelegate(FortMinigameLogic)
  ;[0 1 0.9305 0.1] Minigame State Change Logic
  ;;[0 1 0.9305 0] Turn off hover sphere collision so it doesn't interfere with other collision queries in game
  (Collision|SetCollisionEnabled (Variables|Default|GetHoverOverSphere) "NoCollision")
  (EffectRadius|EffectRadiusRegisterAll))

(event OnComponentBeginOverlap(OverlapCapsule)
  (params (OverlappedComponent :type Primitive%20Component%20Object%20Reference) (OtherActor :type Actor%20Object%20Reference) (OtherComp :type Primitive%20Component%20Object%20Reference) (OtherBodyIndex :type Integer) (bFromSweep :type Boolean) (SweepResult :type Hit%20Result%20Structure%20%28by%20ref%29))
  ;[0 1 0.7928 0.2] Overlap Events
  (CallFunction|TrytoTeleportActorbyOverlap OtherActor))

(event OnComponentEndOverlap(OverlapCapsule)
  (params (OverlappedComponent :type Primitive%20Component%20Object%20Reference) (OtherActor :type Actor%20Object%20Reference) (OtherComp :type Primitive%20Component%20Object%20Reference) (OtherBodyIndex :type Integer))
  ;[0 1 0.7928 0.2] Overlap Events
  ;;[0.15 0 0 0.5] NewtonProps
  (Utilities|Array|RemoveItem (Variables|Default|GetRecentlyTeleportedActors) OtherActor))

(event OnComponentBeginOverlap(EffectRadius)
  (params (OverlappedComponent :type Primitive%20Component%20Object%20Reference) (OtherActor :type Actor%20Object%20Reference) (OtherComp :type Primitive%20Component%20Object%20Reference) (OtherBodyIndex :type Integer) (bFromSweep :type Boolean) (SweepResult :type Hit%20Result%20Structure%20%28by%20ref%29))
  ;[0 1 0.7928 0.2] Overlap Events
  (EffectRadius|OnEffectRadiusOverlappedActor OtherActor true))

(event MinigameEndedDelegate(FortMinigameLogic)
  ;[0 1 0.9305 0.1] Minigame State Change Logic
  ;;[0 1 0.9305 0] Turn off hover sphere collision so it doesn't interfere with other collision queries in game
  (Collision|SetCollisionEnabled (Variables|Default|GetHoverOverSphere) "QueryOnly")
  (EffectRadius|EffectRadiusClear))

(event OnComponentEndOverlap(EffectRadius)
  (params (OverlappedComponent :type Primitive%20Component%20Object%20Reference) (OtherActor :type Actor%20Object%20Reference) (OtherComp :type Primitive%20Component%20Object%20Reference) (OtherBodyIndex :type Integer))
  ;[0 1 0.7928 0.2] Overlap Events
  (EffectRadius|OnEffectRadiusOverlappedActor OtherActor))

(event MinigameStateChangedDelegate(FortMinigameLogic)
  (params (Minigame :type Fort%20Minigame%20Object%20Reference) (NewMinigameState :type EFortMinigameState%20Enum :default "PreGame"))
  ;[0 1 0.9305 0.1] Minigame State Change Logic
  (Default|ResetLink)
  (Variables|Default|SetbShuffledOnce false))

(event Custom|OnPlayModeChanged
  (params (Minigame :type Fort%20Minigame%20Object%20Reference) (bIsInPlayMode :type Boolean))
  ;[0 1 0.9305 0.1] Minigame State Change Logic
  (CallFunction|UpdateEffectRadiusMesh (not bIsInPlayMode))
  (CallFunction|UpdateRiftVisibility))

(event Custom|Enable%20Event
  (params (Instigator :type Controller%20Object%20Reference))
  ;[0 0.6524 1 0.1] Enabled Component Event Handlers
  (Class|CreativeEnabledComponent|SetEnabled (Variables|Default|GetEnabledComponent) true))

(event OnEnabledStateChanged(EnabledComponent)
  (params (Enabled :type Boolean))
  ;[0 1 0.9305 0.1] Enabled State Changing - Handle enable component events, force enable for creation mode
  (Networking|SwitchHasAuthority
    (:Authority
      (if Enabled
        (Variables|VFX|SetTeleporterEventType "Enabled")
        (else
          (Variables|VFX|SetTeleporterEventType "Disabled")))
      (Variables|VFX|SetTeleporterEventFlash (Math|Byte|Byte+Byte (Variables|VFX|GetTeleporterEventFlash) 1)))))

(event Custom|Disable%20Event
  (params (Instigator :type Controller%20Object%20Reference))
  ;[0 0.6524 1 0.1] Enabled Component Event Handlers
  (Class|CreativeEnabledComponent|SetEnabled (Variables|Default|GetEnabledComponent) false))

(event Custom|ActivateLink
  (params (Instigator :type Controller%20Object%20Reference))
  ;[0.5452 0.03075 1 0.1] Gameplay Functions - Linking
  (Variables|Default|SetbLinkToTarget true))

(event Custom|Activate%20Event
  (params (Instigator :type Controller%20Object%20Reference))
  ;[0.5452 0.03075 1 0.1] Gameplay Function - ActivateWhenReceivingFrom triggers the same flow as if the player had overlapped with the teleport
  (Networking|SwitchHasAuthority
    (:Authority
      (if (Class|CreativeEnabledComponent|IsEnabled (Variables|Default|GetEnabledComponent))
        (Utilities|IsValid Instigator
          (:Is%20Valid
            (bind _asfort_player_pawn (Utilities|Casting|CastToFortPlayerPawn (Pawn|GetControlledPawn Instigator)))
            (CallFunction|TrytoTeleportActorbyOverlap _asfort_player_pawn)))))))

(event Custom|DeactivateLink
  (params (Instigator :type Controller%20Object%20Reference))
  ;[0.5452 0.03075 1 0.1] Gameplay Functions - Linking
  (Variables|Default|SetbLinkToTarget false)
  (Default|ResetLink))

(event Custom|ResetLinkEvent
  (params (Instigator :type Controller%20Object%20Reference))
  ;[0.5452 0.03075 1 0.1] Gameplay Functions - Linking
  (Default|ResetLink))

(event Custom|TeleportWhenReceiving
  (params (Instigator :type Controller%20Object%20Reference))
  ;[0.5452 0.03075 1 0.1] Gameplay Function - If a pawn sends a teleport to trigger to this teleporter, teleport them to us
  (if (Class|CreativeEnabledComponent|IsEnabled (Variables|Default|GetEnabledComponent))
    (bind _binprogress (CallFunction|IsMinigameinProgress))
    (if _binprogress
      (Utilities|IsValid Instigator
        (:Is%20Valid
          (bind _asfort_player_controller (Utilities|Casting|CastToFortPlayerController Instigator)
            (:then
              (bind _bcanteleport (Default|CanTeleportActor (Game|Player|GetPlayerPawn _asfort_player_controller)))
              (if _bcanteleport
                (bind _teleportermanagercomponent (CallFunction|GetTeleporterManager))
                (Utilities|IsValid _teleportermanagercomponent
                  (:Is%20Valid
                    (CallFunction|TeleportFromEvent (Game|Player|GetPlayerPawn _asfort_player_controller))))))
            (:CastFailed
              ;[1 0.5621 0.4316] FORT-702506 | Support Teleporter Usage on NPC Characters
              (bind _asaicontroller (Utilities|Casting|CastToAIController Instigator))
              (bind _asfort_pawn (Utilities|Casting|CastToFortPawn (Pawn|GetControlledPawn _asaicontroller)))
              (bind _bcanteleport_2 (Default|CanTeleportActor _asfort_pawn))
              (if _bcanteleport_2
                (bind _teleportermanagercomponent_2 (CallFunction|GetTeleporterManager))
                (Utilities|IsValid _teleportermanagercomponent_2
                  (:Is%20Valid
                    (CallFunction|TeleportNPCPawnFromEvent _asfort_pawn)))))))))))

(event EventEndPlay
  (params (EndPlayReason :type EEndPlayReason%20Enum :default "Destroyed"))
  ;[1 0.1899 0 0.1] Handle Clean-up
  (|CallParentFunction EndPlayReason)
  (Networking|SwitchHasAuthority
    (:Authority
      (bind _teleportermanagercomponent (CallFunction|GetTeleporterManager))
      (Utilities|IsValid _teleportermanagercomponent
        (:Is%20Valid
          (Teleporter|UnregisterTeleporterfromGroup _teleportermanagercomponent self (Variables|Default|GetLastChannel)))))))

(event OnRemovedfromMinigame(FortMinigameLogic)
  (params (Minigame :type Fort%20Minigame%20Object%20Reference))
  ;[1 0.1899 0 0.1] Handle Clean-up
  (Utilities|IsValid Minigame
    (:Is%20Valid
      (Minigame|UnbindEventfromOnPlayModeChangedDelegate Minigame (EventDispatchers|CreateEvent :!event Custom|OnPlayModeChanged)))))

(event HandleVersionUpgrades(CreativeVersion)
  (params (LastSavedVersion :type Creative%20Version%20Id%20Structure))
  ;[1 1 1] Migration for Prop Momentum options released in 38.00
  (Utilities|FlowControl|Sequence
    (:then_0
      (CreativeVersion|CompareVersiontoTag LastSavedVersion "(TagName=\"Creative.Version.Release.38.00\")"
        (:OlderVersion
          (Variables|UserOptions|SetConserve%20Physics%20Prop%20Momentum (Variables|UserOptions|GetConserve%20Momentum))
          (Variables|UserOptions|SetPhysics%20Prop%20Momentum%20Conservation (Utilities|Select :Index (Variables|UserOptions|GetFace%20Physics%20Prop%20in%20Teleporter%20Direction) :NewEnumerator0 "NewEnumerator0" :NewEnumerator1 "NewEnumerator2" :NewEnumerator2 "NewEnumerator1"))
          (Variables|Internal|SetPhysicsPropsOptionsMigrationHandled true))))
    (:then_1
      (if (Variables|Internal|GetPhysicsPropsOptionsMigrationHandled)
        (OptionData|RefreshPropertyOverrides (Variables|Default|GetToyOptionsComponent))))))

(event Custom|PlayTeleportExitedFX
  (params (Location :type Vector))
  (bind _exitriftps (Variables|VFX|GetExitRiftPS))
  (bind _enteredriftsound (Variables|Sound|GetEnteredRiftSound))
  ;[1 0.75 0 0.1] On Teleporter Exited play a SFX
  (Networking|CanExecuteCosmeticEvents
    (:True
      (Utilities|FlowControl|Sequence
        (:then_0
          (if (Variables|UserOptions|Basic|GetPlay%20Visual%20Effects)
            (Utilities|IsValid _exitriftps
              (:Is%20Valid
                (Effects|SpawnEmitteratLocation _exitriftps Location)))))
        (:then_1
          (if (Variables|UserOptions|GetPlay%20Sound%20Effects)
            ;;[1 0.75 0 0.1] Delay to allow for teleported player to hear exit sound
            (Utilities|FlowControl|Delay 0.1)
            (Utilities|IsValid _enteredriftsound
              (:Is%20Valid
                (Audio|PlaySoundatLocation _enteredriftsound Location "0, 0, 0" 1.0 1.0 0.0 0 "/Game/Athena/Sounds/Events/RiftBeacon/RiftBeacon_SingleVoice_Conc.RiftBeacon_SingleVoice_Conc")))))))))

(event Custom|PlayTeleportEnteredFX
  (params (Location :type Vector))
  (bind _enteredriftsound (Variables|Sound|GetEnteredRiftSound))
  ;[1 0.75 0 0.1] On teleporter entered, play SFX and VFX
  (Networking|CanExecuteCosmeticEvents
    (:True
      (Utilities|FlowControl|Sequence
        (:then_0
          (if (Variables|UserOptions|Basic|GetPlay%20Visual%20Effects)
            (Effects|SpawnEmitteratLocation (Variables|VFX|GetEnterRiftPS) Location)))
        (:then_1
          (if (Variables|UserOptions|GetPlay%20Sound%20Effects)
            (Utilities|IsValid _enteredriftsound
              (:Is%20Valid
                (Audio|PlaySoundatLocation _enteredriftsound Location "0, 0, 0" 1.0 1.0 0.0 0 "/Game/Athena/Sounds/Events/RiftBeacon/RiftBeacon_SingleVoice_Conc.RiftBeacon_SingleVoice_Conc")))))))))

(event Custom|PortalEnabledFX
  (bind _staticmeshcomponent (Variables|Debug|GetStaticMeshComponent))
  ;[1 0.75 0 0.1] Open Rift
  (if (or (Variables|UserOptions|Basic|GetPlay%20Visual%20Effects) (Variables|UserOptions|Basic|GetTeleporter%20Rift%20Visible))
    (Networking|CanExecuteCosmeticEvents
      (:True
        (Effects|Components|ParticleSystem|SetTemplate (Variables|VFX|GetP_Rift_Idle_Loop) "/Game/Creative/Devices/Teleporter/P_RefinedRift_Teleport_Idle.P_RefinedRift_Teleport_Idle")
        (CallFunction|InitializeFX)
        (Development|SetHiddeninGame _staticmeshcomponent false false) ; Temporarily show Mesh
        (Rendering|SetVisibility _staticmeshcomponent true false)
        (bind (_direction _linear) (Timeline|TPEnabledFXTimeline PlayFromStart)
          (:Update
            (Rendering|Material|SetCustomPrimitiveDataFloat _staticmeshcomponent 0 _linear))
          (:Finished
            (CallFunction|UpdateRiftVisibility)))))))

(event Custom|PortalDisabledFX
  ;[1 0.75 0 0.1] Close Rift
  (if (or (Variables|UserOptions|Basic|GetPlay%20Visual%20Effects) (Variables|UserOptions|Basic|GetTeleporter%20Rift%20Visible))
    (Networking|CanExecuteCosmeticEvents
      (:True
        (Effects|Components|ParticleSystem|SetTemplate (Variables|VFX|GetP_Rift_Idle_Loop) "/Game/Creative/Items/Gameplay/Teleporter/P_Rift_Portal_Idle.P_Rift_Portal_Idle")
        (bind (_direction _linear) (Timeline|TPDisabledFXTimeline PlayFromStart)
          (:Update
            (Rendering|Material|SetCustomPrimitiveDataFloat (Variables|Debug|GetStaticMeshComponent) 0 _linear))
          (:Finished
            (CallFunction|UpdateRiftVisibility)))))))

(event Custom|Update%20EffectRadiusMesh
  (params (Load :type Boolean))
  (bind _effectradius (Variables|Default|GetEffectRadius))
  ; Disable Effect Radius mesh visibility trace collision on edit mode to handle phone manipulations
  (Collision|SetCollisionResponsetoChannel (Variables|Default|GetEffectRadius) "ECC_Visibility" (select Load "ECR_Ignore" "ECR_Overlap"))
  ;[1 0.01588 0.8999 0.25] Memory Optimization, while we need the mesh during play, it'll be invisible, so we don't need to load the material.
  (if Load
    (Rendering|Material|SetMaterial _effectradius 0 0) ; Clear Portal Effect in Play Mode
    (else
      (bind _loaded_object (Creative|Loading|AsyncLoadEditorSafe (Variables|Default|GetEffectRadiusMaterial))
        (:Loaded
          (bind _asmaterial_interface (Utilities|Casting|CastToMaterialInterface _loaded_object))
          (Rendering|Material|SetMaterial _effectradius 0 _asmaterial_interface) ; Clear Portal Effect in Play Mode
        )))))

(fn BindOrUnbindToPlayerRegisterEvents
  (params (FortPlayerPawnAthena :type Fort%20Player%20Pawn%20Athena%20Object%20Reference%20%28by%20ref%29) (Bind :type Boolean))
  (bind _outputdelegate (EventDispatchers|CreateEvent :!event OnPlayerDied))
  (Utilities|IsValid FortPlayerPawnAthena
    (:Is%20Valid
      (if Bind
        (EventDispatchers|BindEventtoOnFortPlayerDied FortPlayerPawnAthena _outputdelegate)
        (else
          (EventDispatchers|UnbindEventfromOnFortPlayerDied FortPlayerPawnAthena _outputdelegate))))))

(fn BlueprintCanInteract
  (params (InteractingPawn :type Fort%20Pawn%20Object%20Reference) (InteractionBeingAttempted :type EInteractionBeingAttempted%20Enum :default "FirstInteraction") (InteractionType :type TInteractionType%20Enum :default "IT_NoInteraction"))
  (bind _returnvalue (Pawn|GetController InteractingPawn))
  (bind _fortminigamelogic (Variables|Default|GetFortMinigameLogic))
  (bind _currentminigame (Variables|/Script/FortniteGame.FortMinigameLogicComponent|GetCurrentMinigame _fortminigamelogic))
  (Utilities|FlowControl|Sequence
    (:then_0
      (Utilities|IsValid InteractingPawn
        (:Is%20Valid
          (Utilities|IsValid _returnvalue
            (:Is%20Valid
              (bind _asfort_player_controller_athena (Utilities|Casting|CastToFortPlayerControllerAthena _returnvalue))
              (if (.ReturnValue (Creative|CanPlayerCreateinCurrentVolume _asfort_player_controller_athena :!class /Script/FortniteGame.FortPlayerControllerAthena))
                (Utilities|IsValid _currentminigame
                  (:Is%20Valid
                    (LocalVariables|SetbInteractSuccess (not (Minigame|HasStarted _currentminigame :!class /Script/FortniteGame.FortMinigame))))
                  (:Is%20Not%20Valid
                    (LocalVariables|SetbInteractSuccess true)))))))))
    (:then_1
      (return (LocalVariables|GetbInteractSuccess)))))

(fn CanTeleportActor
  (params (Actor%20to%20Teleport :type Actor%20Object%20Reference) (bIgnoreTags :type Boolean))
  (bind _actor_to_teleport (Variables|GetActor%20to%20Teleport))
  (bind _recentlyteleportedactors (Variables|Default|GetRecentlyTeleportedActors))
  (bind _returnvalue (Utilities|Array|ContainsItem _recentlyteleportedactors _actor_to_teleport))
  (bind _enabledcomponent (Variables|Default|GetEnabledComponent))
  (bind _enabled (Class|CreativeEnabledComponent|IsEnabled _enabledcomponent))
  (bind _returnvalue_2 (Utilities|Time|GetTimeSeconds))
  (bind _returnvalue_3 (not _returnvalue))
  (bind _actor_to_teleport_2 (Variables|GetActor%20to%20Teleport))
  (bind _actorslastteleporttime (Variables|Default|GetActorsLastTeleportTime))
  (bind (_value _returnvalue_4) (Utilities|Map|Find _actorslastteleporttime _actor_to_teleport_2))
  (bind _returnvalue_5 (not _returnvalue_4))
  (bind _returnvalue_6 (- _returnvalue_2 _value))
  (bind _returnvalue_7 (> _returnvalue_6 0.5))
  (bind _returnvalue_8 (or _returnvalue_7 _returnvalue_5))
  (Utilities|FlowControl|Sequence
    (:then_0
      (Utilities|IsValid (Variables|GetActor%20to%20Teleport)
        (:Is%20Not%20Valid
          ;[0.15 0.03429 0.1352 0.5] Print Deny Reason
          (PawnUtils|DevPrintTeleportDenyReason "Actor is invalid")
          (return false))))
    (:then_1
      ;[1 1 1] Check to see if the Actor has any blocking gameplay effects that may prevent teleporter use
      (if (PawnUtils|HasActorAnyBlockingTag (Variables|GetActor%20to%20Teleport) (Variables|GetbIgnoreTags))
        ;;[0.15 0.03429 0.1352 0.5] Print Deny Reason
        (PawnUtils|DevPrintTeleportDenyReason "Cooldown teleport is active")
        (return false)))
    (:then_2
      (bind _asfort_pawn (Utilities|Casting|CastToFortPawn (Variables|GetActor%20to%20Teleport))
        (:then
          (Utilities|FlowControl|Sequence
            (:then_0
              ;[1 1 1] Check is player is in creative ghost/phase mode
              (if (PawnUtils|IsPawninCreativeGhost _asfort_pawn)
                (return false)))
            (:then_1
              ;[1 1 1] Check that the player is the right class to use the teleporter
              (bind _valid_controller (CreativeUserRestrictions|ActorHasValidParticipantController (Variables|Default|GetCreative_UseRestriction_Component) _asfort_pawn true true true))
              (if _valid_controller
                ;[1 1 1] Don't allow pawns to teleport if they're in vehicles until V2, unexpected behaviour otherwise
                (Utilities|IsinVehicle _asfort_pawn
                  (:true
                    (return false))
                  (:false
                    (return true)))
                (else
                  ;;[0.15 0.03429 0.1352 0.5] Print Deny Reason
                  (PawnUtils|DevPrintTeleportDenyReason "Actor has not a valid controller")
                  (return false))))))
        (:CastFailed
          ;[1 1 1] Check if Newton Prop
          (bind _returnvalue_9 (FortPhysicsSimulationLibrary|IsPhysicallySimulatedIslandPhysicsProp (Variables|GetActor%20to%20Teleport)))
          (bind _returnvalue_10 (FortPhysicsSimulationLibrary|IsPhysicallySimulatedIslandWeldedProp (Variables|GetActor%20to%20Teleport)))
          (if (Math|Boolean|ANDBoolean (or _returnvalue_10 _returnvalue_9) _returnvalue_8 _returnvalue_3 _enabled)
            ;;[1 1 1] Remove from this teleporter since we know it can teleport.
            (Utilities|Map|Remove (Variables|Default|GetActorsLastTeleportTime) (Variables|GetActor%20to%20Teleport))
            (else
              ;;[0.15 0.03429 0.1352 0.5] Print Deny Reason
              (PawnUtils|DevPrintTeleportDenyReason "Physics prop doesn't match requirements")))
          (return (Math|Boolean|ANDBoolean (or _returnvalue_10 _returnvalue_9) _returnvalue_8 _returnvalue_3 _enabled)))))))

(fn DevPrintTeleportDenyReason
  (params (DenyTeleportReason :type String))
  (if (CallFunction|IsPortalDebuggingEnabled)
    (Development|PrintString DenyTeleportReason true true "(R=0.456733,G=0.960172,B=1.000000,A=1.000000)" 5.0)))

(fn EffectRadius_Clear
  (Networking|SwitchHasAuthority
    (:Authority
      (bind (_arrayindex _arrayelement) (Utilities|Array|ReverseforEachLoop (Variables|Transient|GetEffectRadiusRegisteredActors))
        (:LoopBody
          (EffectRadius|EffectRadiusUnregisterActor _arrayelement))
        (:Completed
          (Utilities|Array|Clear (Variables|Transient|GetEffectRadiusRegisteredActors)))))))

(fn EffectRadius_RegisterActor
  (params (OverlappingActor :type Actor%20Object%20Reference%20%28by%20ref%29))
  (Networking|SwitchHasAuthority
    (:Authority
      (Utilities|IsValid OverlappingActor
        (:Is%20Valid
          (bind _asfort_player_pawn_athena (Utilities|Casting|CastToFortPlayerPawnAthena OverlappingActor)
            (:then
              (if (not (FortPawn|IsDead _asfort_player_pawn_athena))
                _))
            (:CastFailed
              (bind _asbuilding_prop (Utilities|Casting|CastToBuildingProp OverlappingActor))
              (bind _returnvalue (FortPhysicsSimulationLibrary|IsPhysicallySimulatedIslandPhysicsProp _asbuilding_prop))
              (if _returnvalue
                _)))
          (bind _returnvalue_2 (Utilities|Array|AddUnique (Variables|Transient|GetEffectRadiusRegisteredActors) (Variables|GetOverlappingActor)))
          ;[0.1576 1 0 0.5] Event Binding for Athena Pawn
          ;;[1 0.3551 0 0.5] Valid Add - New Actor
          (if (>= _returnvalue_2 0)
            (bind _asfort_player_pawn_athena_2 (Utilities|Casting|CastToFortPlayerPawnAthena (Variables|GetOverlappingActor))
              (:then
                (EffectRadius|BindorUnbindtoPlayerRegisterEvents _asfort_player_pawn_athena_2 true))
              (:CastFailed
                _))
            ;[0.7031 0 0 0.5] Debug
            (if (CallFunction|IsPortalDebuggingEnabled)
              (Development|PrintString (Utilities|String|Append "Registered   Player: " (Utilities|GetDisplayName (Variables|GetOverlappingActor)) " In: " (Utilities|GetDisplayName self)) true true "(R=1.000000,G=0.720884,B=0.660858,A=1.000000)" 1e+01))))))))

(fn EffectRadius_RegisterAll
  (Networking|SwitchHasAuthority
    (:Authority
      ;[1 1 1] Make sure list is empty
      (EffectRadius|EffectRadiusClear)
      (bind _output_get (LocalVariables|SetOverlappingActors (Collision|GetOverlappingActors (Variables|Default|GetEffectRadius) "/Script/FortniteGame.FortPlayerPawn" :!class /Script/Engine.PrimitiveComponent)))
      (for _array_element _output_get
        (EffectRadius|OnEffectRadiusOverlappedActor _array_element true)))))

(fn EffectRadius_RemoveNulls
  (bind (_arrayindex _arrayelement) (Utilities|Array|ReverseforEachLoop (Variables|Transient|GetEffectRadiusRegisteredActors))
    (:LoopBody
      (Utilities|IsValid _arrayelement
        (:Is%20Not%20Valid
          (Utilities|Array|RemoveIndex (Variables|Transient|GetEffectRadiusRegisteredActors) _arrayindex))))))

(fn EffectRadius_UnregisterActor
  (params (OverlappingActor :type Actor%20Object%20Reference%20%28by%20ref%29))
  (Networking|SwitchHasAuthority
    (:Authority
      (Utilities|IsValid OverlappingActor
        (:Is%20Valid
          (bind _returnvalue (Utilities|Array|RemoveItem (Variables|Transient|GetEffectRadiusRegisteredActors) OverlappingActor))
          ;[0.1576 1 0 0.5] Event Binding for Athena Pawn
          ;;[1 0.3551 0 0.5] Valid Removal
          (if _returnvalue
            (bind _asfort_player_pawn_athena (Utilities|Casting|CastToFortPlayerPawnAthena (Variables|GetOverlappingActor))
              (:then
                (EffectRadius|BindorUnbindtoPlayerRegisterEvents _asfort_player_pawn_athena))
              (:CastFailed
                _))
            ;[0.7031 0 0 0.5] Debug
            (if (CallFunction|IsPortalDebuggingEnabled)
              (Development|PrintString (Utilities|String|Append "Unregistered Player: " (Utilities|GetDisplayName (Variables|GetOverlappingActor)) " In: " (Utilities|GetDisplayName self)) true true "(R=1.000000,G=0.720884,B=0.660858,A=1.000000)" 1e+01))))
        (:Is%20Not%20Valid
          (EffectRadius|EffectRadiusRemoveNulls))))))

(fn GetInsideVolumeTargetLocation
  (params (DesiredYaw :type Float%20%28double-precision%29) (ActorToTeleport :type Actor%20Object%20Reference) (TargetTeleporter :type Device%20Teleporter%20V2%20Object%20Reference))
  (bind _teleportlocation (PropUtils|GetTargetLocationtoTeleport DesiredYaw ActorToTeleport TargetTeleporter))
  (bind _returnvalue (Class|CreativeBlueprintLibrary|IsLocationRelevanttoActor (Variables|GetTargetTeleporter) _teleportlocation))
  (if _returnvalue
    (return _teleportlocation)
    (else
      ;[1 1 1] Default Return Teleporter (Target) Location
      (return (Transformation|GetActorLocation (Variables|GetTargetTeleporter))))))

(fn GetTargetLocationToTeleport
  (params (DesiredYaw :type Float%20%28double-precision%29) (ActorToTeleport :type Actor%20Object%20Reference) (TargetTeleporter :type Device%20Teleporter%20V2%20Object%20Reference))
  (bind _targetteleporter (Variables|GetTargetTeleporter))
  (bind _returnvalue (Transformation|GetActorLocation _targetteleporter))
  (bind _actortoteleport (Variables|GetActorToTeleport))
  (bind _returnvalue_2 (Transformation|GetActorLocation _actortoteleport))
  (bind _returnvalue_3 (Transformation|GetActorLocation))
  (bind _returnvalue_4 (- _returnvalue_2 _returnvalue_3))
  (bind _returnvalue_5 (+ _returnvalue _returnvalue_4))
  (bind _zfallbackposition (Variables|Default|GetZFallbackPosition))
  (bind _returnvalue_6 (Math|Vector|vector+vector _returnvalue_5 (Math|Vector|MakeVector 0.0 0.0 _zfallbackposition)))
  ;[0.2307 0 1] Ask teleporter if we should try to maintain relative location
  (if (CallFunction|ShouldMaintainRelativeLocation)
    ;[0.2307 0 1] GetTargetRelativeLocation
    ;[0.2307 0 1] Return on First Location Tried
    (if (not (Teleporter|IsLocationBlocked 0 (Variables|GetActorToTeleport) _returnvalue_5 (Math|Rotator|MakeRotator 0.0 0.0 (Variables|GetDesiredYaw))))
      (return _returnvalue_5)
      (elif (not (Teleporter|IsLocationBlocked 0 (Variables|GetActorToTeleport) _returnvalue_6 (0.0 0.0 0.0)))
        ;[0.2307 0 1] Fallbacked Return
        (return _returnvalue_6)
        (else
          ;[0.2307 0 1] Default Return Teleporter (Target) Location
          (return (Transformation|GetActorLocation (Variables|GetTargetTeleporter))))))
    (else
      ;[0.2307 0 1] Default Return Teleporter (Target) Location
      (return (Transformation|GetActorLocation (Variables|GetTargetTeleporter))))))

(fn GetCreativeActorBounds
  (return true "44.000000,44.000000,44.000000"))

(fn GetCreativeActorOrigin
  (return false "0, 0, 0"))

(fn GetCreativeActorRelativeTransform
  (return false 0))

(fn GetEnabledPhaseIndex
  (return (- (Math|Conversions|ToInteger(Byte) (Variables|UserOptions|GetEnabledDuringPhase)) 1)))

(fn GetExitMomentum_PhysicsProp
  (params (PropToTeleport :type Building%20Prop%20Object%20Reference) (TargetTeleport :type Device%20Teleporter%20V2%20Object%20Reference) (SourceTeleport :type Device%20Teleporter%20V2%20Object%20Reference))
  (bind _proptoteleport (Variables|GetPropToTeleport))
  (bind _returnvalue (FortPhysicsSimulationLibrary|GetSimulatingComponent _proptoteleport))
  (if (Variables|UserOptions|GetConserve%20Physics%20Prop%20Momentum)
    (Utilities|IsValid _returnvalue
      (:Is%20Valid
        (LocalVariables|SetPropAngularVelocity (Physics|GetPhysicsAngularVelocityinRadians _returnvalue "None"))
        (LocalVariables|SetPropLinearVelocity (Transformation|GetComponentVelocity _returnvalue))
        (switch Utilities|FlowControl|Switch|SwitchonETeleporterConserveMomentumProps (Variables|UserOptions|GetPhysics%20Prop%20Momentum%20Conservation)
          (:NewEnumerator0
            ;[1 1 1] No changes at end teleport
            (return (LocalVariables|GetPropLinearVelocity) (LocalVariables|GetPropAngularVelocity)))
          (:NewEnumerator1
            ;[1 1 1] Aligns with Exit Teleporter red arrow
            (return (* (Transformation|GetForwardVector (Variables|Default|GetArrow (Variables|GetTargetTeleport))) (Math|Vector|VectorLength (LocalVariables|GetPropLinearVelocity))) (Math|Vector|RotateVector (LocalVariables|GetPropAngularVelocity) (Math|Rotator|Delta(Rotator) (Transformation|GetActorRotation (Variables|GetPropToTeleport)) (Transformation|GetActorRotation (Variables|GetTargetTeleport))))))
          (:NewEnumerator2
            ;[1 1 1] Keep the angle of entrance as the angle of exit
            (return (Math|Vector|RotateVector (LocalVariables|GetPropLinearVelocity) (Math|Rotator|Delta(Rotator) (Transformation|GetActorRotation (Variables|GetTargetTeleport)) (Transformation|GetActorRotation (Variables|GetSourceTeleport)))) (Math|Vector|RotateVector (LocalVariables|GetPropAngularVelocity) (Math|Rotator|Delta(Rotator) (Transformation|GetActorRotation (Variables|GetTargetTeleport)) (Transformation|GetActorRotation (Variables|GetSourceTeleport))))))))
      (:Is%20Not%20Valid
        _))
    (else
      _))
  ;[1 1 1] Prop stops at the end teleport
  (return "0, 0, 0" "0, 0, 0"))

(fn GetExitRotation_Actor
  (params (ActorToTeleport :type Actor%20Object%20Reference) (TargetTeleporter :type Device%20Teleporter%20V2%20Object%20Reference) (SourceTeleporter :type Device%20Teleporter%20V2%20Object%20Reference))
  (bind _actorrotation_roll (Math|Rotator|BreakRotator (Variables|GetActorRotation_Roll)))
  (Utilities|FlowControl|Sequence
    (:then_0
      (LocalVariables|SetActorRotation (Transformation|GetActorRotation (Variables|GetActorToTeleport))))
    (:then_1
      (bind _asfort_player_pawn (Utilities|Casting|CastToFortPlayerPawn (Variables|GetActorToTeleport)))
      (LocalVariables|SetActorRotation (Pawn|GetControlRotation (Pawn|GetController _asfort_player_pawn) :!class /Script/Engine.Controller)))
    (:then_2
      (return (Math|Rotator|MakeRotator _actorrotation_roll (OrientationUtils|GetTeleportedActorPitch ActorToTeleport SourceTeleporter) (OrientationUtils|GetTeleportedActorYaw ActorToTeleport SourceTeleporter))))))

(fn GetGridAxisGuideAttachmentPoint
  (return false false "0, 0, 0"))

(fn GetPhysicsPropLinearVelocity
  (params (TeleportingActor :type Actor%20Object%20Reference))
  (bind (_origin _boxextent) (Collision|GetActorBounds (LocalVariables|GetActorToTeleport) true false))
  (bind (_boxextent_x _boxextent_y _boxextent_z) (Math|Vector|BreakVector _boxextent))
  (if (and (Utilities|IsValid TeleportingActor :!class /Script/Engine.KismetSystemLibrary) (Utilities|IsValid (Variables|Default|GetDestinationTeleporter) :!class /Script/Engine.KismetSystemLibrary))
    (LocalVariables|SetActorToTeleport TeleportingActor)
    (switch Utilities|FlowControl|Switch|SwitchonEFacePlayerInTeleporterDirection (Variables|UserOptions|GetFace%20Physics%20Prop%20in%20Teleporter%20Direction)
      (:NewEnumerator0
        ;[1 1 1] No option, should keep its original velocity and rotation.
        (LocalVariables|SetFinalLinearVelocity (Math|Vector|SelectVector (Transformation|GetVelocity (LocalVariables|GetActorToTeleport)) "0, 0, 0" (Variables|UserOptions|GetConserve%20Momentum))))
      (:NewEnumerator1
        ;[1 1 1] Relative option , should interpolate rotation to ensure a consistent final velocity with adjacent rotations.
        (LocalVariables|SetFinalLinearVelocity (Math|Vector|SelectVector (Math|Vector|RotateVector (Transformation|GetVelocity (LocalVariables|GetActorToTeleport)) (Math|Rotator|Delta(Rotator) (Transformation|GetActorRotation (Variables|Default|GetDestinationTeleporter)) (Transformation|GetActorRotation))) "0, 0, 0" (Variables|UserOptions|GetConserve%20Momentum))))
      (:NewEnumerator2
        ;[1 1 1] Yes option ,  Aligns with the teleporter rotation-wise , Angular velocity is removed to ensure a smooth and accurate direction
        ;
        (LocalVariables|SetFinalLinearVelocity (* (Math|Vector|SelectVector (Math|Vector|Normalize (Transformation|GetVelocity (LocalVariables|GetActorToTeleport))) (Transformation|GetActorForwardVector (Variables|Default|GetDestinationTeleporter)) (Variables|UserOptions|GetConserve%20Momentum)) (+ (Utilities|Operators|Multiply (/ (Utilities|Operators|Add _boxextent_x _boxextent_y _boxextent_z) 3.0) (Math|Vector|VectorLength (Physics|GetPhysicsAngularVelocityinRadians (FortPhysicsSimulationLibrary|GetSimulatingComponent (LocalVariables|GetActorToTeleport)) "None")) (Variables|Default|GetTotalTransferScaleFactor%20)) (Math|Vector|VectorLength (Transformation|GetVelocity (LocalVariables|GetActorToTeleport))))))
        (LocalVariables|SetShouldResetAngularVel true)))
    (return (LocalVariables|GetFinalLinearVelocity) (LocalVariables|GetShouldResetAngularVel))
    (else
      (return "0, 0, 0" false))))

(fn GetTeleportedActorPitch
  (params (Actor%20to%20Teleport :type Actor%20Object%20Reference) (Source%20Teleporter :type Device%20Teleporter%20V2%20Object%20Reference))
  (Utilities|FlowControl|Sequence
    (:then_0
      (LocalVariables|SetActorRotation (Transformation|GetActorRotation (Variables|GetActor%20to%20Teleport)))
      (LocalVariables|SetbIsPawn false))
    (:then_1
      (bind _asfort_player_pawn (Utilities|Casting|CastToFortPlayerPawn (Variables|GetActor%20to%20Teleport)))
      (LocalVariables|SetActorRotation (Pawn|GetControlRotation (Pawn|GetController _asfort_player_pawn) :!class /Script/Engine.Controller))
      (LocalVariables|SetbIsPawn true))
    (:then_2
      (return (Utilities|Select :Index (select (LocalVariables|GetbIsPawn) (Variables|UserOptions|GetFace%20Player%20In%20Teleporter%20Direction) (Variables|UserOptions|GetFace%20Physics%20Prop%20in%20Teleporter%20Direction)) :NewEnumerator0 (.pitch (LocalVariables|GetActorRotation)) :NewEnumerator1 (+ (.pitch (Math|Rotator|Delta(Rotator) (LocalVariables|GetActorRotation) (Transformation|GetActorRotation (Variables|GetSource%20Teleporter)))) (.pitch (Transformation|GetActorRotation))) :NewEnumerator2 (.pitch (Transformation|GetActorRotation)))))))

(fn GetTeleportedActorYaw
  (params (Actor%20to%20Teleport :type Actor%20Object%20Reference) (Source%20Teleporter :type Device%20Teleporter%20V2%20Object%20Reference))
  (bind (_asfort_pawn _bsuccess) (Utilities|Casting|CastToFortPawn (Variables|GetActor%20to%20Teleport)))
  (return (Utilities|Select :Index (select _bsuccess (Variables|UserOptions|GetFace%20Player%20In%20Teleporter%20Direction) (Variables|UserOptions|GetFace%20Physics%20Prop%20in%20Teleporter%20Direction)) :NewEnumerator0 (.yaw (Transformation|GetActorRotation (Variables|GetActor%20to%20Teleport))) :NewEnumerator1 (+ (.yaw (Math|Rotator|Delta(Rotator) (Transformation|GetActorRotation (Variables|GetActor%20to%20Teleport)) (Transformation|GetActorRotation (Variables|GetSource%20Teleporter)))) (.yaw (Transformation|GetActorRotation))) :NewEnumerator2 (.yaw (Transformation|GetActorRotation)))))

(fn GetTeleporterManager
  (bind _self self)
  (bind _returnvalue (Minigame|GetMinigame _self :!class /Script/FortniteGame.MinigameBlueprintLibrary))
  (Utilities|IsValid _returnvalue
    (:Is%20Valid
      (return (Variables|/Script/FortniteGame.FortMinigame|GetTeleporterManagerComponent _returnvalue)))))

(fn HasActorAnyBlockingTag
  (params (ActorToCheck :type Actor%20Object%20Reference) (bShouldIgnoreTags :type Boolean))
  (bind _actortocheck (Variables|GetActorToCheck))
  (if bShouldIgnoreTags
    (return false)
    (elif (Utilities|DoesObjectImplementInterface _actortocheck "/Script/GameplayTags.GameplayTagAssetInterface")
      (return (GameplayTags|HasAnyMatchingGameplayTags (Utilities|Casting|ToGameplayTagAssetInterface _actortocheck) (Variables|Default|GetTeleporterBlockingTags)))
      (else
        (return true)))))

(fn InitializeFX
  (Rendering|Material|SetCustomPrimitiveDataFloat (Variables|Debug|GetStaticMeshComponent) 0 (select (or (Class|CreativeEnabledComponent|IsEnabled (Variables|Default|GetEnabledComponent)) (and (Editor|InEditor) (Variables|UserOptions|Basic|GetTeleporter%20Rift%20Visible))) 1.0 0.0)))

(fn isMinigameInCreationMode
  (bind _currentminigame (Variables|/Script/FortniteGame.FortMinigameLogicComponent|GetCurrentMinigame (Variables|Default|GetFortMinigameLogic))
    (:then
      (return (not (Minigame|IsinPlayMode _currentminigame :!class /Script/FortniteGame.FortMinigame))))
    (:else
      (return (Editor|InEditor)))))

(fn isMinigameInProgress
  (bind _fortminigamelogic (Variables|Default|GetFortMinigameLogic))
  (bind _currentminigame (Variables|/Script/FortniteGame.FortMinigameLogicComponent|GetCurrentMinigame _fortminigamelogic))
  (Utilities|FlowControl|Sequence
    (:then_0
      (Utilities|IsValid _currentminigame
        (:Is%20Valid
          (LocalVariables|SetbLocalInProgress (Utilities|Enum|Equal(Enum) (Variables|/Script/FortniteGame.FortMinigame|GetCurrentState _currentminigame) "InProgress")))))
    (:then_1
      (return (LocalVariables|GetbLocalInProgress)))))

(fn IsPawnInCreativeGhost
  (params (Pawn :type Fort%20Pawn%20Object%20Reference))
  (bind _asfort_player_pawn_athena (Utilities|Casting|CastToFortPlayerPawnAthena Pawn)
    (:then
      (return (Variables|/Script/FortniteGame.FortPlayerPawnAthena|GetbIsCreativeGhostModeActivated _asfort_player_pawn_athena)))
    (:CastFailed
      (return false))))

(fn MigrateLegacyUserOptions
  (bind (_effectradiusregisteredactors _knob_effectradiussize) (Class|GetClassDefaults "/CreativeCoreDevices/Device_Teleporter_V2.Device_Teleporter_V2_C"))
  (bind _knob_enabledindex (Class|GetClassDefaults "/CreativeCoreDevices/Device_Teleporter_V2.Device_Teleporter_V2_C"))
  (bind _knob_affectsclass (Class|GetClassDefaults "/CreativeCoreDevices/Device_Teleporter_V2.Device_Teleporter_V2_C"))
  (bind _knob_changetarget (Class|GetClassDefaults "/CreativeCoreDevices/Device_Teleporter_V2.Device_Teleporter_V2_C"))
  (bind _knob_changetarget_2 (Variables|LegacyUserOptions|GetKnob_ChangeTarget))
  (bind _knob_enabledindex_2 (Variables|LegacyUserOptions|GetKnob_EnabledIndex))
  (|CallParentFunction)
  (Utilities|FlowControl|Sequence
    (:then_0
      (Utilities|FlowControl|Sequence
        (:then_0
          ;[0 1 0.9331 0.25] Migrating Enabled Index (+1 Shift, -1=Never to 0=Never)
          (if (!= _knob_enabledindex_2 _knob_enabledindex)
            (Variables|UserOptions|SetEnabledDuringPhase (Utilities|Enum|BytetoEnumEEnabledOnPhase (Math|Conversions|ToByte(Integer) (+ _knob_enabledindex_2 1))))
            (Variables|LegacyUserOptions|SetKnob_EnabledIndex _knob_enabledindex)
            (LocalVariables|SetOption%20Migrated true)))
        (:then_1
          ;[0 1 0.9331 0.25] Splitting Old Change Target option into two.
          (if (!= _knob_changetarget _knob_changetarget_2)
            (Variables|UserOptions|SetChangeTeleporterTargetInterval (Math|Float|Max(Float) (Math|Conversions|ToFloat(Integer) _knob_changetarget_2)))
            (Variables|UserOptions|SetChange%20Teleporter%20Target (select (> (Variables|LegacyUserOptions|GetKnob_ChangeTarget) 0) "NewEnumerator3" (Utilities|Select :Index (+ (Variables|LegacyUserOptions|GetKnob_ChangeTarget) 2) :Option%200 "NewEnumerator2" :Option%201 "NewEnumerator0" :Option%202 "NewEnumerator1")))
            (Variables|LegacyUserOptions|SetKnob_ChangeTarget _knob_changetarget)
            (LocalVariables|SetOption%20Migrated true)))
        (:then_2
          ;[1 1 1] Converting Diameter to Radius (it used to be diameter under the hood)
          (if (!= (Variables|LegacyUserOptions|GetKnob_EffectRadiusSize) _knob_effectradiussize)
            (Variables|UserOptions|SetEffect%20Radius (/ (Variables|LegacyUserOptions|GetKnob_EffectRadiusSize) 2.0))
            (Variables|LegacyUserOptions|SetKnob_EffectRadiusSize _knob_effectradiussize)
            (LocalVariables|SetOption%20Migrated true)))
        (:then_3
          ;[0 1 0.9331 0.25] Converting Invert Class Selection, if it's not default, it's on (because default is off, 1)
          (if (!= (Variables|LegacyUserOptions|GetKnob_AffectsClass) _knob_affectsclass)
            (Variables|UserOptions|SetInvert%20Class%20Selection true)
            (Variables|LegacyUserOptions|SetKnob_AffectsClass _knob_affectsclass)
            (LocalVariables|SetOption%20Migrated true)))
        (:then_4
          ;[1 1 1] Converting old 'None' Selected Class to 'Any'
          (if (== (Creative|ClassOption|GetLegacyClassIndex (Variables|UserOptions|GetSelected%20Class)) -2)
            (Variables|SetSelected%20Class_ClassType (struct :ClassType "Any" :ClassSlot 1))
            (Variables|UserOptions|SetInvert%20Class%20Selection (not (Variables|UserOptions|GetInvert%20Class%20Selection)))
            (LocalVariables|SetOption%20Migrated true)))))
    (:then_1
      (if (LocalVariables|GetOption%20Migrated)
        (OptionData|RefreshPropertyOverrides (Variables|Default|GetToyOptionsComponent))))))

(fn OnEffectRadiusOverlappedActor
  (params (OverlappingActor :type Actor%20Object%20Reference%20%28by%20ref%29) (Enter :type Boolean))
  (bind _overlappingactor (Variables|GetOverlappingActor))
  (bind _overlappingactor_2 (Variables|GetOverlappingActor))
  (Networking|SwitchHasAuthority
    (:Authority
      (if (> (Variables|UserOptions|GetEffect%20Radius) 0.0)
        (Utilities|IsValid OverlappingActor
          (:Is%20Valid
            (Utilities|Casting|CastToFortPlayerPawnAthena _overlappingactor_2
              (:then
                _)
              (:CastFailed
                (Utilities|Casting|CastToBuildingProp _overlappingactor_2)))
            (if (Variables|GetEnter)
              (EffectRadius|EffectRadiusRegisterActor _overlappingactor)
              (else
                (EffectRadius|EffectRadiusUnregisterActor _overlappingactor))))
          (:Is%20Not%20Valid
            (EffectRadius|EffectRadiusRemoveNulls)))))))

(fn OnTeleportSucceded
  (params (TargetTeleporter :type Device%20Teleporter%20V2%20Object%20Reference%20%28by%20ref%29) (TriggerInstigator :type Actor%20Object%20Reference%20%28by%20ref%29))
  (Networking|SwitchHasAuthority
    (:Authority
      (if (and (> (Variables|UserOptions|GetEffect%20Radius) 0.0) (not (Variables|Default|GetIgnoreEffectRadiusSettings)))
        ;[1 1 1] This function will be called multiple times recursively. This will avoid that the function does something while this happens
        (if (not (Variables|Transient|GetbIsTeleportingInRadiusPawns))
          (Variables|Transient|SetbIsTeleportingInRadiusPawns true)
          (Utilities|IsValid TriggerInstigator
            (:Is%20Valid
              (bind (_arrayindex _arrayelement) (Utilities|Array|ReverseforEachLoop (Variables|Transient|GetEffectRadiusRegisteredActors))
                (:LoopBody
                  (if (Utilities|NotEqual(Object) TriggerInstigator _arrayelement)
                    (CallFunction|TrytoTeleportActorbyOverlap _arrayelement (Variables|GetTargetTeleporter))))
                (:Completed
                  _)))
            (:Is%20Not%20Valid
              _))
          (Variables|Transient|SetbIsTeleportingInRadiusPawns false))))))

(fn OnPlayerDied
  (params (FortPlayerPawnAthena :type Fort%20Player%20Pawn%20Athena%20Object%20Reference))
  (EffectRadius|EffectRadiusUnregisterActor FortPlayerPawnAthena))

(fn OnReceivedActorClearTimers
  (Utilities|Time|ClearandInvalidateTimerbyHandle (Variables|VFX|GetTeleportEventResetTimer))
  (Variables|VFX|SetTeleporterEventType "Exited")
  (Variables|VFX|SetTeleporterEventFlash (Math|Byte|Byte+Byte (Variables|VFX|GetTeleporterEventFlash) 1))
  (bind _returnvalue (Utilities|Time|SetTimerbyEvent (EventDispatchers|CreateEvent :!event ResetTeleportEventState) 0.1))
  (Variables|VFX|SetTeleportEventResetTimer _returnvalue))

(fn OnReceivedPawn
  (params (PawnReceived :type Fort%20Player%20Pawn%20Object%20Reference))
  (bind _returnvalue (Pawn|GetController PawnReceived))
  (Utilities|IsValid PawnReceived
    (:Is%20Valid
      (Utilities|IsValid _returnvalue
        (:Is%20Valid
          (CallFunction|OnReceivedActorClearTimers)
          (GameplayEvent|BroadcastEvent (Variables|UserOptions|Events|GetOn%20Teleported) 0 (Utilities|Struct|MakeGameplayEventPlayerTeleported _returnvalue)))
        (:Is%20Not%20Valid
          _)))
    (:Is%20Not%20Valid
      _))
  (return))

(fn OnReceivedPhysicsObject
  (params (PhysicsObject :type Actor%20Object%20Reference))
  (Utilities|IsValid PhysicsObject
    (:Is%20Valid
      (Utilities|Map|Add (Variables|Default|GetActorsLastTeleportTime) PhysicsObject (Utilities|Time|GetTimeSeconds))
      ;[1 0.9158 0 0.5] We don't really have a workaround for removing props from this list after a time like we do with pawns
      (if (Collision|IsOverlappingActor (Variables|Default|GetOverlapCapsule) PhysicsObject :!class /Script/Engine.PrimitiveComponent)
        (Utilities|Array|AddUnique (Variables|Default|GetRecentlyTeleportedActors) PhysicsObject)
        (else
          _))
      (CallFunction|OnReceivedActorClearTimers)
      (return))
    (:Is%20Not%20Valid
      (return))))

(fn OnRep_TeleporterUsedFlash
  (Networking|CanExecuteCosmeticEvents
    (:True
      (switch Utilities|FlowControl|Switch|SwitchonEFortCreativeTeleporterEvent (Variables|VFX|GetTeleporterEventType)
        (:Entered
          (CallFunction|PlayTeleportEnteredFX (Transformation|GetActorLocation)))
        (:Exited
          (CallFunction|PlayTeleportExitedFX (Transformation|GetActorLocation)))
        (:Enabled
          (CallFunction|PortalEnabledFX))
        (:Disabled
          (CallFunction|PortalDisabledFX))))))

(fn PawnCancelMomentum
  (params (PlayerPawn :type Fort%20Player%20Pawn%20Object%20Reference))
  (bind _charactermovement (Variables|/Script/Engine.Character|GetCharacterMovement PlayerPawn))
  (Character|StopJumping PlayerPawn)
  (FortMovement|CancelJumpAbility PlayerPawn false)
  (Components|Movement|StopMovementImmediately _charactermovement)
  (Pawn|Components|CharacterMovement|ClearAccumulatedForces _charactermovement)
  (if (Fortnite|IsPhysicallySimulatedIsland)
    (Physics|SetPhysicsLinearVelocity (Variables|/Script/Engine.Character|GetCapsuleComponent PlayerPawn) "0, 0, 0" false "None")))

(fn PawnSetVelocity
  (params (Pawn :type Fort%20Player%20Pawn%20Object%20Reference) (Velocity :type Vector))
  (if (Fortnite|IsPhysicallySimulatedIsland)
    (Physics|SetPhysicsLinearVelocity (Variables|/Script/Engine.Character|GetCapsuleComponent Pawn) Velocity false "None")
    (else
      (Variables|/Script/Engine.MovementComponent|SetVelocity Velocity (Variables|/Script/Engine.Character|GetCharacterMovement Pawn)))))

(fn ResetLink
  (bind _destinationteleporter (Variables|Default|GetDestinationTeleporter)
    (:then
      (if (Variables|Default|GetbIsLinked _destinationteleporter)
        (if (== (Variables|Default|GetDestinationTeleporter _destinationteleporter) self)
          (CallFunction|SetDestinationTeleporter _destinationteleporter 0)
          (CallFunction|SetIsLinked _destinationteleporter false)
          (else
            _))
        (CallFunction|SetDestinationTeleporter 0)
        (bind _binprogress (CallFunction|IsMinigameinProgress))
        (if (not _binprogress)
          (Variables|Default|SetbLinkToTarget (Variables|UserOptions|GetLink%20to%20Target)))
        (CallFunction|SetIsLinked false)
        (else
          _)))
    (:else
      _)))

(fn ResetTeleportEventState
  ;[1 1 1] To prevent initial clients getting bombarded with enter and exit VFX we set the event type to none
  (Networking|SwitchHasAuthority
    (:Authority
      (switch Utilities|FlowControl|Switch|SwitchonEFortCreativeTeleporterEvent (Variables|VFX|GetTeleporterEventType)
        (:Entered
          _)
        (:Exited
          _))
      (Variables|VFX|SetTeleporterEventType "None"))))

(fn SelectDestinationTeleporter
  (params (ActorToTeleport :type Actor%20Object%20Reference))
  (bind _fortminigamelogic (Variables|Default|GetFortMinigameLogic))
  (bind _currentminigame (Variables|/Script/FortniteGame.FortMinigameLogicComponent|GetCurrentMinigame _fortminigamelogic))
  (bind _returnvalue (Minigame|GetRuntime _currentminigame))
  (bind _changeteleportertargetinterval (Variables|UserOptions|GetChangeTeleporterTargetInterval))
  (bind _returnvalue_2 (/ _returnvalue _changeteleportertargetinterval))
  (bind _returnvalue_3 (Math|Float|Floor _returnvalue_2))
  (bind _destinationteleporter (Variables|Default|GetDestinationTeleporter))
  (if (Variables|Default|GetbLinkToTarget)
    (Utilities|IsValid _destinationteleporter
      (:Is%20Valid
        (if (Variables|Default|GetbLinkToTarget _destinationteleporter)
          (if (Variables|Default|GetbIsLinked _destinationteleporter)
            (if (== (Variables|Default|GetDestinationTeleporter _destinationteleporter) self)
              (return _destinationteleporter)
              (else
                _))
            (else
              _))
          (LocalVariables|SetLocalActorToTeleport ActorToTeleport)
          (bind _teleportermanagercomponent (CallFunction|GetTeleporterManager))
          (Utilities|IsValid _teleportermanagercomponent
            (:Is%20Valid
              (bind _result (Utilities|Set|ToArray (Teleporter|GetTeleportersinGroup _teleportermanagercomponent (Variables|UserOptions|Basic|GetKnob_TargetTeleporterGroup))))
              (if (> (Utilities|Array|Length _result) 0)
                (if (Utilities|Enum|Equal(Enum) (Variables|UserOptions|GetChange%20Teleporter%20Target) "NewEnumerator2")
                  (if (not (Variables|Default|GetbShuffledOnce))
                    (Utilities|Array|Shuffle _result)
                    (Variables|Default|SetbShuffledOnce true)
                    (else
                      _))
                  (else
                    (if (Utilities|Enum|Equal(Enum) (Variables|UserOptions|GetChange%20Teleporter%20Target) "NewEnumerator0")
                      ;[1 1 1] Pick the first in the list
                      (for _array_element _result
                        (LocalVariables|SetCurrentTeleporterCandidate _array_element)
                        ;[1 1 1] Loop Body
                        (bind _currentteleportercandidate (LocalVariables|GetCurrentTeleporterCandidate)
                          (:then
                            (bind _asdevice_teleporter_v2 (Utilities|Casting|CastToDevice_Teleporter_V2 _currentteleportercandidate))
                            (if (and (Class|CreativeEnabledComponent|IsEnabled (Variables|Default|GetEnabledComponent _asdevice_teleporter_v2)) (Collision|GetActorEnableCollision _asdevice_teleporter_v2))
                              (if (Utilities|NotEqual(Object) self _asdevice_teleporter_v2)
                                ;[1 1 1] If we intend to pair with the new destination teleporter, make sure we are not breaking an already existing pair
                                (if (Variables|Default|GetbLinkToTarget)
                                  (if (Variables|Default|GetbIsLinked _asdevice_teleporter_v2)
                                    _
                                    (else
                                      _))
                                  (else
                                    _))
                                (if (not (Teleporter|IsTeleporterBlocked _asdevice_teleporter_v2 (LocalVariables|GetLocalActorToTeleport) (Math|Rotator|MakeRotator (.roll (Transformation|GetActorRotation (LocalVariables|GetLocalActorToTeleport))) (.pitch (Transformation|GetActorRotation (LocalVariables|GetLocalActorToTeleport))) (OrientationUtils|GetTeleportedActorYaw (LocalVariables|GetLocalActorToTeleport) self))))
                                  (Variables|Default|SetCurrentTimeSelectedTarget _asdevice_teleporter_v2)
                                  (return _asdevice_teleporter_v2)))))
                          (:else)))
                      (return 0)
                      (else
                        ;[1 1 1] According to design all teleporter behaviours should operate on a random ordering of the teleporter list
                        (Utilities|Array|Shuffle _result)
                        (if (Utilities|Enum|Equal(Enum) (Variables|UserOptions|GetChange%20Teleporter%20Target) "NewEnumerator1")
                          ;[1 1 1] On teleporter entered select the next random teleporter
                          (for _index (range (+ (Utilities|Array|Length _result) 1))
                            (LocalVariables|SetCurrentTeleporterCandidate (Utilities|Array|Get(acopy) _result (Math|Integer|%25%28Integer%29 (+ _index (Variables|Default|GetTeleporterUseCounter)) (Utilities|Array|Length _result)))))
                          (return 0)
                          (else
                            ;[1 1 1] On teleporter entered select based on how much time has passed and how often we should be switching
                            (Utilities|IsValid _currentminigame
                              (:Is%20Valid
                                (if (!= _returnvalue_3 (Variables|Default|GetCurrentTimeSelectedTargetIndex))
                                  (Variables|Default|SetCurrentTimeSelectedTargetIndex _returnvalue_3)
                                  (for _index_2 (range (+ (Utilities|Array|Length _result) 1))
                                    (LocalVariables|SetCurrentTeleporterCandidate (Utilities|Array|Get(acopy) _result (Math|Integer|%25%28Integer%29 (+ _index_2 (Variables|Default|GetCurrentTimeSelectedTargetIndex)) (Utilities|Array|Length _result)))))
                                  (return 0)
                                  (else
                                    (return (Variables|Default|GetCurrentTimeSelectedTarget))))))))))))
                (else
                  (return 0)))))
          (else
            _)))
      (:Is%20Not%20Valid
        _))
    (else
      _)))

(fn SetAndSelectDestinationTeleporter
  (params (Forced%20Destination%20Teleport :type Device%20Teleporter%20V2%20Object%20Reference) (ActorToTeleport :type Actor%20Object%20Reference))
  (Utilities|IsValid Forced%20Destination%20Teleport
    (:Is%20Valid
      (bind _newdestinationteleporter (CallFunction|SetNewDestination Forced%20Destination%20Teleport)))
    (:Is%20Not%20Valid
      (bind _selectedteleporter (CallFunction|SelectDestinationTeleporter ActorToTeleport))
      (bind _newdestinationteleporter_2 (CallFunction|SetNewDestination _selectedteleporter))))
  (return (or (Utilities|IsValid _newdestinationteleporter :!class /Script/Engine.KismetSystemLibrary) (Utilities|IsValid _newdestinationteleporter_2 :!class /Script/Engine.KismetSystemLibrary))))

(fn SetDestinationTeleporter
  (params (destination :type Device%20Teleporter%20V2%20Object%20Reference))
  (bind _output_get (Variables|Default|SetDestinationTeleporter destination))
  (return _output_get))

(fn SetIsLinked
  (params (NewIsLinkedValue :type Boolean))
  (Variables|Default|SetbIsLinked NewIsLinkedValue))

(fn SetNewDestination
  (params (NewDestination :type Device%20Teleporter%20V2%20Object%20Reference))
  (bind _self self)
  (if (Variables|Default|GetbLinkToTarget)
    (bind _newdestination (Variables|GetNewDestination)
      (:then
        (if (Variables|Default|GetbLinkToTarget _newdestination)
          (if (Variables|Default|GetbIsLinked _newdestination)
            (if (== (Variables|Default|GetDestinationTeleporter _newdestination) _self)
              (CallFunction|SetDestinationTeleporter _newdestination _self)
              (CallFunction|SetIsLinked true)
              (CallFunction|SetIsLinked _newdestination true)
              (else
                (CallFunction|SetIsLinked false)))
            (bind _destination_out (CallFunction|SetDestinationTeleporter _newdestination))
            (return _destination_out)
            (else
              _))
          (else
            _)))
      (:else
        _))
    (else
      _)))

(fn ShouldMaintainRelativeLocation
  (return (Math|Boolean|ANDBoolean (Variables|UserOptions|GetMaintain%20Relative%20Position) (> (Variables|UserOptions|GetEffect%20Radius) 0.0) (not (Variables|Default|GetIgnoreEffectRadiusSettings)))))

(fn TeleportfromEvent
  (params (PlayerPawn :type Fort%20Player%20Pawn%20Object%20Reference))
  (bind _charactermovement (Variables|/Script/Engine.Character|GetCharacterMovement PlayerPawn))
  (switch Utilities|FlowControl|Switch|SwitchonE_TeleporterDevice_TeleportToEventMomentum (Variables|UserOptions|Advanced|GetMaintainMomentumDuringTeleportToEvent)
    (:NewEnumerator0
      _)
    (:NewEnumerator1
      (Variables|Default|SetCachedPendingVelocity (Pawn|Character|GetPendingLaunchVelocity PlayerPawn))
      (Pawn|Components|CharacterMovement|ClearAccumulatedForces _charactermovement)
      (PawnUtils|PawnSetVelocity PlayerPawn (* (Math|Float|SelectFloat (Math|Vector|VectorLength (Variables|Default|GetCachedPendingVelocity)) (Math|Vector|VectorLength (Variables|/Script/Engine.MovementComponent|GetVelocity _charactermovement)) (Math|Vector|NotEqualExactly(Vector) (Variables|Default|GetCachedPendingVelocity))) (Math|Vector|Normalize (Transformation|GetActorForwardVector)))))
    (:NewEnumerator2
      (PawnUtils|PawnCancelMomentum PlayerPawn)
      (CallFunction|TeleportActorTo PlayerPawn self true true))))

(fn TeleportNPCPawnfromEvent
  (params (NPC%20Pawn :type Fort%20Pawn%20Object%20Reference))
  (bind (_returnvalue_location _returnvalue_rotation) (Math|Transform|BreakTransform (Transformation|GetActorTransform self)))
  (bind _charactermovement (Variables|/Script/Engine.Character|GetCharacterMovement NPC%20Pawn))
  (bind _charactermovement_2 (Variables|/Script/Engine.Character|GetCharacterMovement NPC%20Pawn))
  (switch Utilities|FlowControl|Switch|SwitchonE_TeleporterDevice_TeleportToEventMomentum (Variables|UserOptions|Advanced|GetMaintainMomentumDuringTeleportToEvent)
    (:NewEnumerator0
      _)
    (:NewEnumerator1
      (Variables|Default|SetCachedPendingVelocity (Pawn|Character|GetPendingLaunchVelocity NPC%20Pawn))
      (Pawn|Components|CharacterMovement|ClearAccumulatedForces _charactermovement)
      (Variables|/Script/Engine.MovementComponent|SetVelocity (* (Math|Float|SelectFloat (Math|Vector|VectorLength (Variables|Default|GetCachedPendingVelocity)) (Math|Vector|VectorLength (Variables|/Script/Engine.MovementComponent|GetVelocity _charactermovement)) (Math|Vector|NotEqualExactly(Vector) (Variables|Default|GetCachedPendingVelocity))) (Math|Vector|Normalize (Transformation|GetActorForwardVector))) (Variables|/Script/Engine.Character|GetCharacterMovement NPC%20Pawn)))
    (:NewEnumerator2
      (Character|StopJumping NPC%20Pawn)
      (Components|Movement|StopMovementImmediately _charactermovement_2)
      (Pawn|Components|CharacterMovement|ClearAccumulatedForces _charactermovement_2)
      (Transformation|Teleport NPC%20Pawn _returnvalue_location _returnvalue_rotation))))

(fn TeleportActorTo
  (params (ActorToTeleport :type Actor%20Object%20Reference) (TargetTeleporter :type Device%20Teleporter%20V2%20Object%20Reference) (bNewIgnoreEffectRadiusSettings%20 :type Boolean) (IgnoreCollision :type Boolean))
  (Variables|Default|SetIgnoreEffectRadiusSettings bNewIgnoreEffectRadiusSettings%20)
  (Ability|SendGameplayEventtoActor self "(TagName=\"Abilities.Generic.ActiveAbility.Activate\")" (Utilities|Struct|MakeGameplayEventData "(TagName=\"Abilities.Generic.ActiveAbility.Activate\")" self ActorToTeleport TargetTeleporter 0 "()" "(GameplayTags=)" (select IgnoreCollision "(GameplayTags=((TagName=\"Gameplay.Collision.Pawn.Ignore\")))" 0))))

(fn TeleportPhysicalProp_PT
  (params (PhysicsObject :type Actor%20Object%20Reference))
  (bind _self self)
  (bind _destinationteleporter (Variables|Default|GetDestinationTeleporter))
  (bind _outrotation (OrientationUtils|GetExitRotationActor PhysicsObject _destinationteleporter _self))
  (bind _returnvalue (Actor|GetComponentbyClass (Variables|GetPhysicsObject) "/Script/FortniteGame.FortPhysicsComponent"))
  (bind _physicsobject (Variables|GetPhysicsObject))
  (bind _teleportlocation (PropUtils|GetInsideVolumeTargetLocation (.yaw _outrotation) PhysicsObject (Variables|Default|GetDestinationTeleporter)))
  (Default|OnReceivedPhysicsObject (Variables|Default|GetDestinationTeleporter) _physicsobject)
  (bind _asbuilding_prop (Utilities|Casting|CastToBuildingProp _physicsobject))
  (CallFunction|ApplySyncedTeleportandVelocityonPT _asbuilding_prop (Math|Transform|MakeTransform _teleportlocation _outrotation) (OrientationUtils|GetExitMomentumPhysicsProp _asbuilding_prop (Variables|Default|GetDestinationTeleporter) self) (.Angular%20Momentum (OrientationUtils|GetExitMomentumPhysicsProp _asbuilding_prop (Variables|Default|GetDestinationTeleporter) self)))
  (Utilities|IsValid _returnvalue
    (:Is%20Valid
      (Class|FortPhysics|NotifyPhysicsPropTeleported _returnvalue))))

(fn TryToTeleportActorByOverlap
  (params (Actor%20to%20Teleport :type Actor%20Object%20Reference%20%28by%20ref%29) (Forced%20Destination%20Teleport :type Device%20Teleporter%20V2%20Object%20Reference))
  (bind _actor_to_teleport (Variables|GetActor%20to%20Teleport))
  (bind (_asfort_pawn _bsuccess) (Utilities|Casting|CastToFortPawn _actor_to_teleport))
  (bind _self self)
  (bind _fortminigamelogic (Variables|Default|GetFortMinigameLogic))
  (bind _currentminigame (Variables|/Script/FortniteGame.FortMinigameLogicComponent|GetCurrentMinigame _fortminigamelogic))
  (bind _volume (Variables|/Script/FortniteGame.FortMinigame|GetVolume _currentminigame))
  (bind _destinationteleporter (Variables|Default|GetDestinationTeleporter))
  (Networking|SwitchHasAuthority
    (:Authority
      (bind _bcanteleport (Default|CanTeleportActor Actor%20to%20Teleport))
      (if _bcanteleport
        (bind _setvaliddestinationteleporter (Default|SetandSelectDestinationTeleporter (Variables|GetForced%20Destination%20Teleport) (Variables|GetActor%20to%20Teleport)))
        (if _setvaliddestinationteleporter
          (bind _returnvalue (Class|CreativeBlueprintLibrary|IsLocationRelevanttoActor _volume (Transformation|GetActorLocation (Variables|Default|GetDestinationTeleporter))))
          (if (and _returnvalue (Utilities|IsValid _volume :!class /Script/Engine.KismetSystemLibrary))
            (Utilities|FlowControl|Sequence
              (:then_0
                ;[1 1 1] VFX Timer
                (Utilities|Time|ClearandInvalidateTimerbyHandle (Variables|VFX|GetTeleportEventResetTimer))
                (Variables|VFX|SetTeleporterEventType "Entered")
                (Variables|VFX|SetTeleporterEventFlash (Math|Byte|Byte+Byte (Variables|VFX|GetTeleporterEventFlash) 1))
                (bind _returnvalue_2 (Utilities|Time|SetTimerbyEvent (EventDispatchers|CreateEvent :!event ResetTeleportEventState) 0.1))
                (Variables|VFX|SetTeleportEventResetTimer _returnvalue_2))
              (:then_1
                (Utilities|FlowControl|Sequence
                  (:then_0
                    ;[0.7344 0.39 0 0.5] Teleport Pawn
                    (bind _asfort_player_pawn (Utilities|Casting|CastToFortPlayerPawn (Variables|GetActor%20to%20Teleport)))
                    (Utilities|FlowControl|Sequence
                      (:then_0
                        (if (Variables|UserOptions|GetConserve%20Momentum)
                          ;;[0 0.09319 1 0.5] Pawn - Conserve Momemtum
                          (PawnUtils|PawnSetVelocity _asfort_player_pawn (Math|Vector|RotateVector (Transformation|GetVelocity (Variables|GetActor%20to%20Teleport)) (Math|Rotator|Delta(Rotator) (Transformation|GetActorRotation (Variables|Default|GetDestinationTeleporter)) (Transformation|GetActorRotation))))
                          (else
                            ;;[0.15 0.04297 0.04297 0.5] Pawn - No Momemtum
                            (PawnUtils|PawnCancelMomentum _asfort_player_pawn))))
                      (:then_1
                        (CallFunction|TeleportActorTo self (Variables|GetActor%20to%20Teleport) (Variables|Default|GetDestinationTeleporter) false false))))
                  (:then_1
                    ;[0.03538 0.15 0.08266 0.5] Newton Physics Props
                    (bind _asbuilding_prop (Utilities|Casting|CastToBuildingProp (Variables|GetActor%20to%20Teleport)))
                    (bind _returnvalue_3 (Development|GetConsoleVariableBoolValue "FortCreativeTeleporter.HandleTeleportOnPT"))
                    (if _returnvalue_3
                      ;;[0.7031 0 0 0.5] Schedule Teleport to be handled on PT
                      (CallFunction|TeleportPhysicalPropPT _asbuilding_prop)
                      (else
                        ;;[0.15 0.1387 0 0.5] Teleport immediately using Gameplay Ability
                        (CallFunction|TeleportActorTo self (Variables|GetActor%20to%20Teleport) (Variables|Default|GetDestinationTeleporter) false false)
                        (FortPhysicsSimulationLibrary|SetPhysicsLinearVelocity _asbuilding_prop (OrientationUtils|GetExitMomentumPhysicsProp _asbuilding_prop _destinationteleporter _self) false "None" 0 "(MaxSimSizeAllowed=Large,bRejectEmbeddedInLandscape=True)")
                        (FortPhysicsSimulationLibrary|SetPhysicsAngularVelocity _asbuilding_prop (.Angular%20Momentum (OrientationUtils|GetExitMomentumPhysicsProp _asbuilding_prop _destinationteleporter _self)) false "None" 0 "(MaxSimSizeAllowed=Large,bRejectEmbeddedInLandscape=True)"))))))
              (:then_2
                ;[0.000287 0.7667 0.8438 0.5] After Teleport
                (Utilities|FlowControl|Sequence
                  (:then_0
                    (if _bsuccess
                      (GameplayEvent|BroadcastEvent (Variables|UserOptions|Events|GetOn%20Enter) 0 (Utilities|Struct|MakeGameplayEventPlayerEntered (Pawn|GetController _asfort_pawn)))))
                  (:then_1
                    (bind _result (Math|Integer|IncrementInt (Variables|Default|GetTeleporterUseCounter))
                      (:%20%20
                        (Variables|Default|SetTeleporterUseCounter _result)))))))))))))

(fn UpdateEffectRadiusScale
  (bind (_relativescale3d_x _relativescale3d_y _relativescale3d_z) (Math|Vector|BreakVector (Variables|/Script/Engine.SceneComponent|GetRelativeScale3D_X (Variables|Default|GetEffectRadius))))
  (Transformation|SetRelativeScale3D (Variables|Default|GetEffectRadius) (Math|Vector|MakeVector (Math|Float|Max(Float) (Math|Float|float*float (Variables|UserOptions|GetEffect%20Radius) (Variables|Default|GetM%20To%20CMConversion) 2.0) 0.01) (Math|Float|Max(Float) (Math|Float|float*float (Variables|UserOptions|GetEffect%20Radius) (Variables|Default|GetM%20To%20CMConversion) 2.0) 0.01) _relativescale3d_z))
  (Development|SetHiddeninGame (Variables|Default|GetEffectRadius) (Math|Float|NearlyEqual(Float) (Variables|UserOptions|GetEffect%20Radius)) true))

(fn UpdatePhysicsObjectCollisionSettings
  ;[0 1 0.3723 0.1] Enable/Disable collision with physics objects depending on the setting
  (Collision|SetCollisionResponsetoChannel (Variables|Default|GetOverlapCapsule) "ECC_WorldStatic" (select (Variables|UserOptions|Basic|GetCan%20Teleport%20Physics%20Props) "ECR_Overlap" "ECR_Ignore"))
  (return))

(fn UpdateRiftVisibility
  (bind _visible_in_game (LocalVariables|GetVisible%20In%20Game))
  (bind _returnvalue (not _visible_in_game))
  (bind _portal_inworld_loop_cue (Variables|Default|GetPortal_InWorld_Loop_Cue))
  ;[1 1 1] If the game phase allows visibility and our teleporter is set to be visible or we're in creation mode, show the rift mesh and VFX
  (Networking|CanExecuteCosmeticEvents
    (:True
      (bind _bincreationmode (CallFunction|IsMinigameinCreationMode))
      (bind _output_get (LocalVariables|SetVisible%20In%20Game (and (Variables|UserOptions|Basic|GetTeleporter%20Rift%20Visible) (or (Class|CreativeEnabledComponent|IsEnabled (Variables|Default|GetEnabledComponent)) (Editor|InEditor)))))
      (Rendering|SetVisibility (Variables|Debug|GetStaticMeshComponent) _output_get false)
      (Rendering|SetVisibility (Variables|VFX|GetP_Rift_Idle_Loop) (or _output_get _bincreationmode) false)
      (Development|SetHiddeninGame (Variables|Debug|GetStaticMeshComponent) _returnvalue false) ; The move tool is forcing visibility on for \na reason I can't determine when first placed.\nForcing Hidden in game if not visible to \nkeep this off also.
      (Development|SetHiddeninGame (Variables|VFX|GetP_Rift_Idle_Loop) (and _returnvalue (Editor|InEditor)) false) ; Idle  effects are only hidden if they're not visible in game\nAND are in editor. This means they're normally not hidden \n(so mainly controlled by Visibility) but in Editor will hide \nwhen in Game Mode.
      (if (LocalVariables|GetVisible%20In%20Game)
        (Audio|Components|Audio|Play _portal_inworld_loop_cue 0.0)
        (else
          (Audio|Components|Audio|Stop _portal_inworld_loop_cue))))))

(fn ConstructionScript)
