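; Creative timer device that runs a countdown or count-up timer, either per player or shared globally or by team.
; Channel events and interaction drive an Enabled/Activated/Paused/Completed/Disabled state machine, filtered by team and class.
; The authority tracks timers across minigame start and end, awarding score and lap times, with optional persistent save/load for personal timers.
; Clients reconcile from server finish time to update world text, HUD and urgency effects.
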(event OnAnyPropertyChanged(ToyOptionsComponent)
  ;[0 1 0.09771 0.25] Initialization
  (Delegates|BindDelegates)
  (CallFunction|BindClientExitVolumeGameplayMessage)
  (Initialization|Initialize (Variables|Default|GetCreative_UseRestriction_Component) (Variables|Components|GetFortMinigameProgress) 0 0 (Variables|UserOptions|GetActivatingTeam) false (Variables|UserOptions|GetActivatingClass) false false)
  (Variables|Default|SetDeviceTeamID (Creative|TeamOption|ConverttoTeam (Variables|UserOptions|GetActivatingTeam)))
  (Visual|UpdateinGameVisibilities)
  (Class|CreativeGenericTextWidgetComp|UpdateText (Variables|Default|GetPrimaryTextWidget) (Variables|UserOptions|GetTimerName))
  (CallFunction|InitializeTimerText)
  (Variables|FortCreativeTimerDevice|SetClientTimerTextStyle (Utilities|Map|Find (Variables|Default|GetTextStyleToStringMap) (Variables|UserOptions|GetTimerLabelTextStyle)))
  (Variables|FortCreativeTimerDevice|SetbClientShowOnHud (Variables|UserOptions|GetShowOnHUD))
  (Class|CreativeUIBroadcastComponent|Initialize (Variables|Default|GetCreative_HUDBroadcast_Component) (Variables|Components|GetFortMinigameProgress) (Variables|UserOptions|Advanced|GetDisplayScoreUpdateOnHUD))
  (Class|CreativeTimerLibrary|SetupTimerUI))

(event OnPropertiesLoaded(ToyOptionsComponent)
  ;[0 1 0.09771 0.25] Initialization
  (Delegates|BindDelegates)
  (CallFunction|BindClientExitVolumeGameplayMessage)
  (Initialization|Initialize (Variables|Default|GetCreative_UseRestriction_Component) (Variables|Components|GetFortMinigameProgress) 0 0 (Variables|UserOptions|GetActivatingTeam) false (Variables|UserOptions|GetActivatingClass) false false)
  (Variables|Default|SetDeviceTeamID (Creative|TeamOption|ConverttoTeam (Variables|UserOptions|GetActivatingTeam)))
  (Visual|UpdateinGameVisibilities)
  (Class|CreativeGenericTextWidgetComp|UpdateText (Variables|Default|GetPrimaryTextWidget) (Variables|UserOptions|GetTimerName))
  (CallFunction|InitializeTimerText)
  (Variables|FortCreativeTimerDevice|SetClientTimerTextStyle (Utilities|Map|Find (Variables|Default|GetTextStyleToStringMap) (Variables|UserOptions|GetTimerLabelTextStyle)))
  (Variables|FortCreativeTimerDevice|SetbClientShowOnHud (Variables|UserOptions|GetShowOnHUD))
  (Class|CreativeUIBroadcastComponent|Initialize (Variables|Default|GetCreative_HUDBroadcast_Component) (Variables|Components|GetFortMinigameProgress) (Variables|UserOptions|Advanced|GetDisplayScoreUpdateOnHUD))
  (Class|CreativeTimerLibrary|SetupTimerUI))

(event EventBeginPlay
  ;[0 1 0.09771 0.25] Initialization
  (|CallParentFunction)
  (Networking|SwitchHasAuthority
    (:Authority
      (Persistence|InitializePersistenceComponent (Variables|Default|GetFortCreativePersistenceComponentBase) (Variables|Components|GetFortMinigameProgress) true)
      (bind _returnvalue (Game|ConstructObjectfromClass (Variables|Persistence|GetPersistenceData_Timer_Class) self))
      (Variables|Persistence|SetPersistenceData_Timer _returnvalue))))

(event EditorPreview|EventSetupForEditorPreview
  ;[0 1 0.09771 0.25] Initialization
  ;;[0 0.5156 0.1638 0] Editor Only - When moving, we need to reconstruct the timer widget every frame as the references will get disconnected. This gets called every frame and does that.
  (|CallParentFunction)
  (Class|CreativeGenericTextWidgetComp|UpdateText (Variables|Default|GetPrimaryTextWidget) (Variables|UserOptions|GetTimerName)))

(event EventEndPlay
  (params (EndPlayReason :type EEndPlayReason%20Enum :default "Destroyed"))
  ;[1 0.08089 0 0.25] Destruction Cleanup
  (|CallParentFunction EndPlayReason)
  (Delegates|UnbindDelegates)
  (Class|CreativeTimerLibrary|CleanupTimerUI))

(event EventClearDisplay
  ;[1 1 1] Client side reconcilation from server
  (Networking|CanExecuteCosmeticEvents
    (:True
      (Visual|ClearDisplayTexts 0 true)
      (Variables|Cosmetic|SetbUrgencyHasStarted false))))

(event Custom|StartTimerFromEvent
  (params (Instigator :type Controller%20Object%20Reference))
  ;[0.3706 0 1 0.25] Channel Receivers
  (bind _output_get (Variables|Default|SetLastInstigator Instigator))
  (Utilities|IsValid _output_get
    (:Is%20Valid
      (Variables|Default|SetActivatingPlayer _output_get))
    (:Is%20Not%20Valid
      _))
  (bind (_bexists _trackedplayer) (Default|GetIntendedTrackedPlayer _output_get))
  (if _bexists
    (switch Utilities|FlowControl|Switch|SwitchonETimerDeviceState (CallFunction|GetTimerState _trackedplayer)
      (:Enabled
        (CallFunction|TimerSetState _trackedplayer "Activated")))))

(event EventDisplayUpdate
  (params (CurrentState :type Byte) (ServerTimeFinished :type Float%20%28single-precision%29))
  (bind _returnvalue (Fortnite|Sync|Time|GetServerWorldTimeSeconds))
  (bind _returnvalue_2 (- ServerTimeFinished _returnvalue))
  (bind _returnvalue_3 (Utilities|Enum|BytetoEnumETimerDeviceState CurrentState))
  ;[1 1 1] Client side reconcilation from server
  (Networking|CanExecuteCosmeticEvents
    (:True
      (Visual|UpdateinGameDisplay _returnvalue_3 _returnvalue_2)
      (Default|SetupClientDisplayUpdateTimer _returnvalue_3 _returnvalue_2))))

(event Custom|CompleteTimerFromEvent
  (params (Instigator :type Controller%20Object%20Reference))
  ;[0.3706 0 1 0.25] Channel Receivers
  (bind _output_get (Variables|Default|SetLastInstigator Instigator))
  (bind (_bexists _trackedplayer) (Default|GetIntendedTrackedPlayer _output_get))
  (if _bexists
    (switch Utilities|FlowControl|Switch|SwitchonETimerDeviceState (CallFunction|GetTimerState _trackedplayer)
      (:Enabled
        _)
      (:Activated
        _)
      (:Paused
        _))
    (CallFunction|TimerSetState _trackedplayer "Completed")))

(event Custom|ResetTimerFromEvent
  (params (Instigator :type Controller%20Object%20Reference))
  ;[0.3706 0 1 0.25] Channel Receivers
  (bind _output_get (Variables|Default|SetLastInstigator Instigator))
  (bind (_bexists _trackedplayer) (Default|GetIntendedTrackedPlayer _output_get))
  (if _bexists
    (CallFunction|ResetTimer _trackedplayer)))

(event Custom|CompletionStateChangeDelay
  (params (TrackedPlayer :type Controller%20Object%20Reference) (NewState :type ETimerDeviceState%20Enum :default "Enabled"))
  (CallFunction|TimerClearHandles TrackedPlayer)
  (Utilities|FlowControl|Delay (Variables|Default|GetCompletionDelay))
  (CallFunction|TimerSetState TrackedPlayer NewState))

(event Custom|PauseTimerFromEvent
  (params (Instigator :type Controller%20Object%20Reference))
  ;[0.3706 0 1 0.25] Channel Receivers
  (bind (_bexists _trackedplayer) (Default|GetIntendedTrackedPlayer Instigator))
  (if _bexists
    (switch Utilities|FlowControl|Switch|SwitchonETimerDeviceState (CallFunction|GetTimerState _trackedplayer)
      (:Activated
        (CallFunction|TimerSetState _trackedplayer "Paused")))))

(event Custom|ResumeTimerFromEvent
  (params (Instigator :type Controller%20Object%20Reference))
  ;[0.3706 0 1 0.25] Channel Receivers
  (bind _output_get (Variables|Default|SetLastInstigator Instigator))
  (bind (_bexists _trackedplayer) (Default|GetIntendedTrackedPlayer _output_get))
  (if _bexists
    (switch Utilities|FlowControl|Switch|SwitchonETimerDeviceState (CallFunction|GetTimerState _trackedplayer)
      (:Paused
        (CallFunction|TimerSetState _trackedplayer "Activated")))))

(event Interaction|EventBlueprintOnInteract
  (params (InteractingPawn :type Fort%20Pawn%20Object%20Reference) (InteractionBeingAttempted :type EInteractionBeingAttempted%20Enum :default "FirstInteraction"))
  ;[0 1 0.9331 0.25] Interact Logic. If the timer is running, complete it. Else Start it (if we're not disabled).
  (|CallParentFunction InteractingPawn InteractionBeingAttempted)
  (bind _output_get (Variables|Default|SetLastInstigator (Pawn|GetController InteractingPawn)))
  (bind (_bexists _trackedplayer) (Default|GetIntendedTrackedPlayer _output_get))
  (if _bexists
    (if (Utilities|Enum|Equal(Enum) (CallFunction|GetTimerState _trackedplayer) "Completed")
      (CallFunction|TimerClearHandles _trackedplayer))
    (if (Utilities|Enum|Equal(Enum) (Utilities|Select :Index (CallFunction|GetTimerState _trackedplayer) :Enabled "Activated" :Activated "Completed" :Paused "Activated" :Completed "Activated" :Disabled "Disabled") "Activated")
      (Variables|Default|SetActivatingPlayer (Variables|Default|GetLastInstigator))
      (else
        _))
    (CallFunction|TimerSetState _trackedplayer (Utilities|Select :Index (CallFunction|GetTimerState _trackedplayer) :Enabled "Activated" :Activated "Completed" :Paused "Activated" :Completed "Activated" :Disabled "Disabled"))))

(event Custom|EnableTimerFromEvent
  (params (Instigator :type Controller%20Object%20Reference))
  ;[0.3706 0 1 0.25] Channel Receivers
  (bind (_bexists _trackedplayer) (Default|GetIntendedTrackedPlayer Instigator))
  (if _bexists
    (switch Utilities|FlowControl|Switch|SwitchonETimerDeviceState (CallFunction|GetTimerState _trackedplayer)
      (:Disabled
        (CallFunction|TimerSetState _trackedplayer)))))

(event Custom|DisableTimerFromEvent
  (params (Instigator :type Controller%20Object%20Reference))
  ;[0.3706 0 1 0.25] Channel Receivers
  (bind (_bexists _trackedplayer) (Default|GetIntendedTrackedPlayer Instigator))
  (if _bexists
    (CallFunction|DisableTimer _trackedplayer)))

(event Custom|SetLapTimeFromEvent
  (params (Instigator :type Controller%20Object%20Reference))
  (bind _bexists (Default|GetIntendedTrackedPlayer Instigator))
  (if _bexists
    (CallFunction|SetLapTimeForPlayer Instigator)))

(event Custom|OnMinigameStarted
  (bind _fortminigameprogress (Variables|Components|GetFortMinigameProgress))
  (bind _currentminigame (Variables|/Script/FortniteGame.FortMinigameLogicComponent|GetCurrentMinigame _fortminigameprogress))
  (Variables|Default|SetLastInstigator 0)
  (Minigame|BindEventtoPlayerChangedTeamDelegate _currentminigame (EventDispatchers|CreateEvent :!event OnTeamChanged))
  (Minigame|BindEventtoOnPlayerClassApplied _currentminigame (EventDispatchers|CreateEvent :!event On%20Class%20Applied))
  (Minigame|BindEventtoOnPlayerJoininProgress _currentminigame (EventDispatchers|CreateEvent :!event OnPlayerJoinInProgress))
  (Default|InitializeTrackedTimers)
  (Networking|SwitchHasAuthority
    (:Authority
      (Utilities|FlowControl|Sequence
        (:then_0
          (if (and (Utilities|Enum|Equal(Enum) (Variables|UserOptions|GetAppliesTo) "NewEnumerator0") (Variables|UserOptions|GetUsePersistence))
            (if (and (Utilities|Enum|Equal(Enum) (Variables|UserOptions|GetAuto-LoadWhen) "NewEnumerator1") (Variables|UserOptions|GetAuto-Load))
              (CallFunction|LoadTimerforAll (Variables|UserOptions|GetLoadElapsedTime))
              (else
                _))
            (Persistence|SetAutoSaveEnabled (Variables|Default|GetFortCreativePersistenceComponentBase) (and (Variables|UserOptions|GetAuto-Save) (Utilities|Enum|Equal(Enum) (Variables|UserOptions|GetAuto-SaveWhen) "NewEnumerator1")))))
        (:then_1
          (bind _returnvalue (GameplayEventRouter|GetEventRouter "/Script/FortniteGame.FortPlayspace" self)) ; ContextActor is how we bind to a specific playspace
          (bind (_asynctaskproxy _payload) (GameplayEventRouter|StartListeningToEvent _returnvalue "/Script/FortniteGame.FortTeamChangedContext")
            (:then
              (Variables|Default|SetTeamChangedObserver _asynctaskproxy))
            (:OnEventReceived
              (bind (_payload_player _payload_teamid) (Utilities|Struct|BreakFortTeamChangedContext _payload))
              (Delegates|OnTeamChanged _payload_player _payload_teamid 0 true))) ; Context here is irrelevant
        )))))

(event Custom|StartAllTimersFromEvent
  (params (Instigator :type Controller%20Object%20Reference))
  ;[0.4319 0 1] Channel Receivers for triggering all personal timers
  (Default|UpdateStateforAll Instigator "Activated" (Utilities|Set|MakeSet "Enabled")))

(event Custom|CompleteAllTimersFromEvent
  (params (Instigator :type Controller%20Object%20Reference))
  ;[0.4319 0 1] Channel Receivers for triggering all personal timers
  (Default|UpdateStateforAll Instigator "Completed" (Utilities|Set|MakeSet "Activated" "Paused" "Enabled")))

(event Custom|SetLapTimeForAllEvent
  (params (Instigator :type Controller%20Object%20Reference))
  (bind _bexists (Default|GetIntendedTrackedPlayer Instigator))
  (if _bexists
    (CallFunction|SetLapTimeForPlayer 0)))

(event Custom|PauseAllTimersFromEvent
  (params (Instigator :type Controller%20Object%20Reference))
  ;[0.4319 0 1] Channel Receivers for triggering all personal timers
  (Default|UpdateStateforAll Instigator "Paused" (Utilities|Set|MakeSet "Activated")))

(event Custom|ResetAllTimersFromEvent
  (params (Instigator :type Controller%20Object%20Reference))
  ;[0.4319 0 1] Channel Receivers for triggering all personal timers
  (Default|ResetTimerforAll Instigator))

(event Custom|ResumeAllTimersFromEvent
  (params (Instigator :type Controller%20Object%20Reference))
  ;[0.4319 0 1] Channel Receivers for triggering all personal timers
  (Default|UpdateStateforAll Instigator "Activated" (Utilities|Set|MakeSet "Paused")))

(event Custom|OnMinigameEnded
  (bind _fortminigameprogress (Variables|Components|GetFortMinigameProgress))
  (bind _currentminigame (Variables|/Script/FortniteGame.FortMinigameLogicComponent|GetCurrentMinigame _fortminigameprogress))
  (Networking|SwitchHasAuthority
    (:Authority
      (Variables|Default|SetLastInstigator 0)
      (Minigame|UnbindEventfromPlayerChangedTeamDelegate _currentminigame (EventDispatchers|CreateEvent :!event OnTeamChanged))
      (Minigame|UnbindEventfromOnPlayerClassApplied _currentminigame (EventDispatchers|CreateEvent :!event On%20Class%20Applied))
      (Minigame|UnbindEventfromOnPlayerJoininProgress _currentminigame (EventDispatchers|CreateEvent :!event OnPlayerJoinInProgress))
      (CallFunction|ClearTrackedTimers)
      (GameplayEventRouter|StopListeningtoEvent (Variables|Default|GetTeamChangedObserver)))
    (:Remote
      (Visual|ClearDisplayTexts 0 false))))

(event Custom|SaveTimerFromEvent
  (params (Instigator :type Controller%20Object%20Reference))
  (bind (_outstate _returnvalue) (CallFunction|GetTimerState Instigator))
  ;[1 0 0.7152 0.2] Channel Receivers for Save/Load
  (Persistence|SaveTimers Instigator)
  (if (Math|Boolean|ANDBoolean (Variables|UserOptions|GetUsePersistence) (not (Variables|UserOptions|GetLoadElapsedTime)) _returnvalue (Utilities|Enum|Equal(Enum) _outstate "Activated"))
    (CallFunction|TimerSetState Instigator "Paused")))

(event Custom|LoadTimerFromEvent
  (params (Instigator :type Controller%20Object%20Reference))
  ;[1 0 0.7152 0.2] Channel Receivers for Save/Load
  (Persistence|LoadTimer Instigator))

(event Custom|ClearPersistenceDataFromEvent
  (params (Instigator :type Controller%20Object%20Reference))
  ;[1 0 0.7152 0.2] Channel Receivers for Save/Load
  (bind _bresult (Default|CanTriggerAllPersonalTimers Instigator)) ; Same conditions used to check if timer is valid for personal timer
  (if (and _bresult (Variables|UserOptions|GetUsePersistence))
    (Persistence|ClearData (Variables|Default|GetFortCreativePersistenceComponentBase) (PlayerState|GetUniqueNetId (Variables|/Script/Engine.Controller|GetPlayerState Instigator)) (Variables|Persistence|GetSavePath))
    (GameplayEvent|BroadcastEvent (Variables|UserOptions|Events|GetOn%20Cleared) 0 (Utilities|Struct|MakeGameplayEventTimeronCleared Instigator))))

(event Custom|ClearPersistenceDataForAllFromEvent
  (params (Instigator :type Controller%20Object%20Reference))
  ;[1 0 0.7152 0.2] Channel Receivers for Save/Load
  (bind _bresult (Default|CanTriggerAllPersonalTimers Instigator))
  (if (and (Variables|UserOptions|GetUsePersistence) _bresult)
    (bind _playerstate (Utilities|ForEachPersistentPlayer)
      (:LoopBody
        (Persistence|ClearData (Variables|Default|GetFortCreativePersistenceComponentBase) (PlayerState|GetUniqueNetId _playerstate) (Variables|Persistence|GetSavePath))
        (GameplayEvent|BroadcastEvent (Variables|UserOptions|Events|GetOn%20Cleared) 0 (Utilities|Struct|MakeGameplayEventTimeronCleared (CreativeAI|Controller|GetParticipantControllerfromActor _playerstate false)))))))

(event Custom|SaveTimerForAllFromEvent
  (params (Instigator :type Controller%20Object%20Reference))
  ;[1 0 0.7152 0.2] Channel Receivers for Save/Load
  (if (Variables|UserOptions|GetUsePersistence)
    (bind _playerstate (Utilities|ForEachPersistentPlayer)
      (:LoopBody
        (Persistence|SaveTimers (CreativeAI|Controller|GetParticipantControllerfromActor _playerstate false))
        (if (Math|Boolean|ANDBoolean (Variables|UserOptions|GetUsePersistence) (not (Variables|UserOptions|GetLoadElapsedTime)) (.ReturnValue (CallFunction|GetTimerState (CreativeAI|Controller|GetParticipantControllerfromActor _playerstate false))) (Utilities|Enum|Equal(Enum) (CallFunction|GetTimerState (CreativeAI|Controller|GetParticipantControllerfromActor _playerstate false)) "Activated"))
          (CallFunction|TimerSetState (CreativeAI|Controller|GetParticipantControllerfromActor _playerstate false) "Paused"))))))

(event Custom|LoadTimerForAllFromEvent
  (params (Instigator :type Controller%20Object%20Reference))
  ;[1 0 0.7152 0.2] Channel Receivers for Save/Load
  (CallFunction|LoadTimerforAll false))

(event OnAddedtoMinigame(FortMinigameProgress)
  (params (Minigame :type Fort%20Minigame%20Object%20Reference))
  (Minigame|BindEventtoOnPlayModeChangedDelegate Minigame (EventDispatchers|CreateEvent :!event Custom|On%20Play%20Mode%20Changed))
  (Networking|SwitchHasAuthority
    (:Authority
      ;[1 1 1] Player removed needed so we can clear reference if the referenced player leaves the game
      (Minigame|BindEventtoPlayerRemovedDelegate Minigame (EventDispatchers|CreateEvent :!event OnPlayerRemovedFromMinigame))
      (Minigame|BindEventtoPlayerLastChancetoSave Minigame (EventDispatchers|CreateEvent :!event On%20Last%20Chance%20to%20Save)))))

(event OnRemovedfromMinigame(FortMinigameProgress)
  (params (Minigame :type Fort%20Minigame%20Object%20Reference))
  (Networking|SwitchHasAuthority
    (:Authority
      (Minigame|UnbindEventfromPlayerRemovedDelegate Minigame (EventDispatchers|CreateEvent :!event OnPlayerRemovedFromMinigame))
      (Minigame|UnbindEventfromPlayerLastChancetoSave Minigame (EventDispatchers|CreateEvent :!event On%20Last%20Chance%20to%20Save)))))

(event Custom|On%20Play%20Mode%20Changed
  (params (Minigame :type Fort%20Minigame%20Object%20Reference) (bIsInPlayMode :type Boolean))
  ;[1 0 0.6342] When Play Mode Changes, disable the collision box if we don't want to use it.
  (Collision|SetCollisionEnabled (Variables|Default|GetInteractionBox) (select (and bIsInPlayMode (Utilities|Enum|Equal(Enum) (Variables|UserOptions|GetCanInteract) "NewEnumerator0")) "NoCollision" "QueryOnly"))
  (Visual|UpdateinGameVisibilities)
  (CallFunction|InitializeTimerText))

(event MinigameEndedDelegate(FortMinigameProgress)
  (Networking|SwitchHasAuthority
    (:Authority
      ;[1 1 1] Clear Last Instigator
      (Variables|Default|SetLastInstigator 0)
      ;[1 1 1] Clear Any Overrides
      (Variables|Verse|SetDurationOverride -1.0))))

(event OnVisibilityChanged(VisibleInGameComponent)
  (params (Visible :type Boolean))
  (Class|CreativeGenericTextWidgetComp|UpdateVisibilityandUpdateWidget (Variables|Default|GetPrimaryTextWidget) (Utilities|Select :Index (Variables|UserOptions|GetVisibleDuringGame) :NewEnumerator0 false :NewEnumerator1 false :NewEnumerator2 Visible)))

(event Custom|BindClientExitVolumeGameplayMessage
  ;[1 1 1] Bind to Gameplay Message for exit volume
  (bind (_asynctaskproxy _payload) (Messaging|RegisterGameplayMessageReceiver "(TagName=\"EventMessage.Component.Playspace.PlayerExitedSpatialActor.LegacyMinigame\")" "/Script/FortniteGame.PlayerExitSpatialActorContextLegacyMinigame")
    (:then
      (Variables|Default|SetExitedEventHandle _asynctaskproxy))
    (:OnMessageReceived
      (Delegates|OnClientExitVolume (.ExitingPlayerState (Utilities|Struct|BreakPlayerExitSpatialActorContextLegacyMinigame _payload)) (Utilities|Struct|BreakPlayerExitSpatialActorContextLegacyMinigame _payload)))))

(event Custom|TriggerUrgentFX
  (bind _returnvalue_x (Math|Vector|BreakVector (Rendering|Components|TextRender|GetTextWorldSize (Variables|Components|GetTimerDisplayTextFront))))
  (bind _visibleduringgame (Variables|UserOptions|GetVisibleDuringGame))
  (bind _returnvalue (Utilities|Enum|NotEqual(Enum) _visibleduringgame "NewEnumerator0"))
  (Utilities|FlowControl|Sequence
    (:then_0
      (if (not (Variables|Cosmetic|GetbUrgencyHasStarted))
        (Variables|Cosmetic|SetbUrgencyHasStarted true)
        (Visual|UpdateTimerColor (Variables|Default|GetUrgencyModeColor))
        (if (Math|Boolean|ANDBoolean (not (Audio|Components|Audio|IsPlaying (Variables|Default|GetUrgentBuildAudio))) (Variables|UserOptions|GetAudioEffects) (or (Variables|UserOptions|GetShowOnHUD) _returnvalue))
          (Audio|Components|Audio|Play (Variables|Default|GetUrgentBuildAudio) 0.0))))
    (:then_1
      (if (and (Utilities|IsValid (Variables|Default|GetUrgencySound) :!class /Script/Engine.KismetSystemLibrary) (Variables|UserOptions|GetAudioEffects))
        (if (Variables|UserOptions|GetShowOnHUD)
          (Audio|PlaySound2D (Variables|Default|GetUrgencySound) 1.0 1.0 0.0 0 self false)
          (elif _returnvalue
            (Audio|PlaySoundatLocation (Variables|Default|GetUrgencySound) (Transformation|GetActorLocation) (Transformation|GetActorRotation) 1.0 1.0 0.0 0 0 self)))))
    (:then_2
      (Rendering|Material|SetScalarParameterValueonMaterials (Variables|Default|GetUrgencyGlow) (Variables|Default|GetUrgencyGlowScaleXParameter) (* _returnvalue_x (Variables|Default|GetUrgencyGlowScaleMultiplier)))
      (CallFunction|FlashVFXTimeline))))

(event Custom|FlashVFXTimeline
  (bind (_direction _lerp) (Timeline|Timeline_0 PlayFromStart)
    (:Update
      (Rendering|Material|SetScalarParameterValueonMaterials (Variables|Default|GetUrgencyGlow) (Variables|Default|GetUrgencyGlowOpacityParameter) _lerp))))

(fn AddTime
  (params (Instigator :type Controller%20Object%20Reference) (AdditionalTime :type Float%20%28double-precision%29))
  (Networking|SwitchHasAuthority
    (:Authority
      (bind (_hasactivetimer _timeremaining) (CallFunction|GetRemainingTime Instigator))
      (if _hasactivetimer
        (CallFunction|TimerStart Instigator (Math|Float|Clamp(Float) (+ _timeremaining AdditionalTime) 1.0 (Utilities|GetCurrentDuration)) (- (Math|Float|Clamp(Float) (+ _timeremaining AdditionalTime) 1.0 (Utilities|GetCurrentDuration)) (Variables|UserOptions|GetUrgencyModeTime)))
        ;[1 1 1] If the timer isn't running (activated) immediately pause to avoid starting the timer
        (if (not (Utilities|Enum|Equal(Enum) (CallFunction|GetTimerState Instigator) "Activated"))
          (CallFunction|TimerPause Instigator))
        (CallFunction|RequestServerUpdateForPlayer Instigator)))))

(fn AwardStats
  (params (TrackedPlayer :type Fort%20Player%20Controller%20Object%20Reference) (TimeRemaining :type Float%20%28double-precision%29))
  (bind (_outsourcetags _outcontexttags) (Class|FortMinigameStatsComponent|GetSourceandContextTags (LocalVariables|Getlocal%20MinigameStatsComp) (Variables|GetTrackedPlayer)))
  (Utilities|IsValid TrackedPlayer
    (:Is%20Valid
      (bind _output_get (LocalVariables|SetSuccess (CallFunction|TimerWasSuccessful TimeRemaining)))
      (if (and _output_get (Variables|UserOptions|GetSetLapTimeOnSuccess))
        (CallFunction|SetLapTimeForPlayer (Variables|GetTrackedPlayer)))
      (bind _returnvalue (CreativeDevices|GetMinigameStatsComp (Variables|GetTrackedPlayer) "Athena"))
      (LocalVariables|Setlocal%20MinigameStatsComp _returnvalue)
      (LocalVariables|Getlocal%20MinigameStatsComp
        (:then
          (bind _output_get_2 (LocalVariables|SetScore (+ (Math|Integer|SelectInt (Variables|UserOptions|GetSuccessScoreValue) (Variables|UserOptions|GetFailureScorePenalty) (LocalVariables|GetSuccess)) (Math|Integer|SelectInt (Math|Float|Truncate (* (Math|Float|Truncate (Variables|GetTimeRemaining)) (Variables|UserOptions|GetScorePerSecondRemaining))) 0 (> (Variables|GetTimeRemaining) 0.0)))))
          (if (!= _output_get_2 0)
            (bind _output_get_3 (LocalVariables|SetTarget%20Tags (GameplayTags|GetOwnedGameplayTags :!class /Script/GameplayTags.GameplayTagAssetInterface)))
            (GameplayTags|AddGameplayTag _output_get_3 "(TagName=\"Minigame.Stat.Score.Goal\")" :!class /Script/GameplayTags.BlueprintGameplayTagLibrary)
            (Class|FortMinigameStatsComponent|SendStatEventwithTags (LocalVariables|Getlocal%20MinigameStatsComp) (Variables|GetTrackedPlayer) "MinigameDynamicEvent" self (LocalVariables|GetTarget%20Tags) _outsourcetags _outcontexttags (LocalVariables|GetScore))
            (elif (and (Variables|UserOptions|Advanced|GetDisplayScoreUpdateIfScoreIs0) (Variables|UserOptions|Advanced|GetDisplayScoreUpdateOnHUD))
              _))
          (Class|CreativeHUDBroadcastComponent|DoScoreEvent (Variables|Default|GetCreative_HUDBroadcast_Component) (Variables|GetTrackedPlayer) (LocalVariables|GetScore) (Variables|UserOptions|Advanced|GetDisplayScoreUpdateOnHUD)))
        (:else)))))

(fn BindDelegates
  (bind _fortminigameprogress (Variables|Components|GetFortMinigameProgress))
  (Minigame|BindEventtoMinigameStateChangedDelegate _fortminigameprogress (EventDispatchers|CreateEvent :!event OnMinigameStateChanged))
  (Minigame|BindEventtoMinigameStartedDelegate _fortminigameprogress (EventDispatchers|CreateEvent :!event OnMinigameStarted))
  (Minigame|BindEventtoMinigameEndedDelegate _fortminigameprogress (EventDispatchers|CreateEvent :!event OnMinigameEnded)))

(fn BlueprintCanInteract
  (params (InteractingPawn :type Fort%20Pawn%20Object%20Reference) (InteractionBeingAttempted :type EInteractionBeingAttempted%20Enum :default "FirstInteraction") (InteractionType :type TInteractionType%20Enum :default "IT_NoInteraction"))
  (if (not (Interaction|CanEditOptionsComponent (Variables|Components|GetToyOptionsComponent) InteractingPawn))
    (if (Utilities|Enum|NotEqual(Enum) (Variables|UserOptions|GetCanInteract) "NewEnumerator0")
      (bind _interactingpawn (Variables|GetInteractingPawn)
        (:then
          (bind (_valid_controller _fortplayerpawn _controller) (CreativeUserRestrictions|ActorHasValidParticipantController (Variables|Default|GetCreative_UseRestriction_Component) _interactingpawn true true true))
          (if _valid_controller
            (Networking|SwitchHasAuthority
              (:Authority
                (LocalVariables|Setlocal%20State (CallFunction|GetTimerState _controller)))
              (:Remote
                (LocalVariables|Setlocal%20State (Variables|FortCreativeTimerDevice|GetClientCurrentState))))
            (return (Utilities|Select :Index (LocalVariables|Getlocal%20State) :Enabled (Utilities|Enum|NotEqual(Enum) (Variables|UserOptions|GetCanInteract) "NewEnumerator2") :Activated (Utilities|Enum|NotEqual(Enum) (Variables|UserOptions|GetCanInteract) "NewEnumerator1") :Paused true :Completed (Utilities|Enum|NotEqual(Enum) (Variables|UserOptions|GetCanInteract) "NewEnumerator2") :Disabled false))))
        (:else))
      (else
        (return false)))
    (else
      (return true))))

(fn BlueprintGetInteractionString
  (params (InteractingPawn :type Fort%20Pawn%20Object%20Reference) (InteractionBeingAttempted :type EInteractionBeingAttempted%20Enum :default "FirstInteraction"))
  (if (Interaction|CanEditOptionsComponent (Variables|Components|GetToyOptionsComponent) InteractingPawn)
    (return "")
    (else
      (return (Utilities|Text|FormatText "NSLOCTEXT(\"[91C0351C4A570D194F4D469F4C93D163]\", \"AA04010740C9FCA8BD13698C65366948\", \"{action} {timername}{time}\")" (Utilities|Select :Index (Variables|FortCreativeTimerDevice|GetClientCurrentState) :Enabled (Variables|Default|GetTimerStartText) :Activated (Variables|Default|GetTimerCompleteText) :Paused (Variables|Default|GetTimerStartText) :Completed (Variables|Default|GetTimerStartText)) (select (Utilities|Text|TextIsEmpty (Variables|UserOptions|GetTimerName)) (Variables|Default|GetTimerDefaultName) (Variables|UserOptions|GetTimerName)) (select (Utilities|Enum|Equal(Enum) (Variables|FortCreativeTimerDevice|GetClientCurrentState) "Activated") (Utilities|Select :Index (Variables|UserOptions|Basic|GetCountDownDirection) :NewEnumerator0 (Utilities|Text|FormatText "NSLOCTEXT(\"[91C0351C4A570D194F4D469F4C93D163]\", \"05D902F54C8E89E2FB619FB502E18EF1\", \"\\r\\n{time}s remaining\")" (Variables|FortCreativeTimerDevice|GetClientCurrentTime)) :NewEnumerator7 (Utilities|Text|FormatText "NSLOCTEXT(\"[91C0351C4A570D194F4D469F4C93D163]\", \"0CF6B0E14EC22C1B0598A5B3C43859B4\", \"\\r\\n{time}s\")" (Variables|FortCreativeTimerDevice|GetClientCurrentTime))) ""))))))

(fn BlueprintGetInteractionTime
  (params (InteractingPawn :type Fort%20Pawn%20Object%20Reference) (InteractionBeingAttempted :type EInteractionBeingAttempted%20Enum :default "FirstInteraction"))
  (return (Math|Float|SelectFloat (Variables|UserOptions|GetInteractTime) 0.0 (Class|CreativeCommonDeviceFunctionLibrary|IsLogicinPlayMode (Variables|Components|GetFortMinigameProgress))) (Utilities|Enum|NotEqual(Enum) (Variables|UserOptions|GetCanInteract) "NewEnumerator0")))

(fn CanClassInteract
  (params (Player :type Controller%20Object%20Reference))
  (return (StateChecks|IsValidClass (Variables|Default|GetCreative_UseRestriction_Component) Player)))

(fn CanPlayerAutoSave
  (params (FortPlayerState :type Fort%20Player%20State%20Object%20Reference))
  (bind (_outstate _returnvalue) (CallFunction|GetTimerState (Fortnite|GetControllerfromActor FortPlayerState)))
  (return (Math|Boolean|ANDBoolean (or (Creative|TeamOption|IsTeamType (Variables|UserOptions|GetActivatingTeam) "Any") (Math|Byte|Equal(Byte) (Variables|Default|GetDeviceTeamID) (Team|GetActorTeam (Fortnite|GetControllerfromActor FortPlayerState)))) (CallFunction|CanClassInteract (Fortnite|GetControllerfromActor FortPlayerState)) _returnvalue)))

(fn CanTriggerAllPersonalTimers
  (params (Instigator :type Controller%20Object%20Reference))
  (Utilities|IsValid Instigator
    (:Is%20Valid
      (LocalVariables|SetTeamActorCheck (Math|Byte|Equal(Byte) (Team|GetActorTeam Instigator) (Variables|Default|GetDeviceTeamID))) ; Only check actor team if instagator is valid\n(avoid a null actor team check)
    )
    (:Is%20Not%20Valid
      _))
  (return (Math|Boolean|ANDBoolean (Utilities|Enum|Equal(Enum) (Variables|UserOptions|GetAppliesTo) "NewEnumerator0") (or (Creative|TeamOption|IsTeamType (Variables|UserOptions|GetActivatingTeam) "Any") (LocalVariables|GetTeamActorCheck)) (CallFunction|CanClassInteract Instigator))))

(fn CheckFailureDisableOrReset
  (params (TrackedPlayer :type Controller%20Object%20Reference))
  (if (Variables|UserOptions|Advanced|GetDisableTimerIfFailingTeamOrClassCheck)
    (CallFunction|DisableTimer TrackedPlayer)
    (elif (Variables|UserOptions|Advanced|GetResetTimerIfFailingTeamOrClassCheck)
      (CallFunction|ResetTimer TrackedPlayer))))

(fn ClearDisplayTexts
  (params (ClearTimeDisplay :type Boolean))
  (bind (_customuserwidgetclass _timeralignment _teamchangedobserver _clientcurrenttime) (Class|GetClassDefaults (Utilities|GetClass self)))
  (if ClearTimeDisplay
    (Rendering|Components|TextRender|SetText (Variables|Components|GetTimerDisplayTextFront) (Utilities|Text|FormatText "")) ; Timer Device must update these text values to clear the text on device's widget
    (else
      _))
  (Utilities|Time|ClearandInvalidateTimerbyHandle (Variables|Cosmetic|GetClientDisplayUpdateTimerHandle))
  (Variables|FortCreativeTimerDevice|SetClientCurrentSecondaryText "")
  (Class|CreativeGenericTextWidgetComp|UpdateText (Variables|Default|GetSecondaryTextWidget) "")
  (Visual|UpdateTimerColor (Variables|UserOptions|GetTimerColor))
  ;[0.8129 0.6457 1] Reset display time back to class default to allow inactive timers to have their worldspace text widget initially display on subsequent minigames
  (Variables|FortCreativeTimerDevice|SetClientCurrentTime _clientcurrenttime)
  (if (Variables|UserOptions|GetShowOnHUD)
    (CallFunction|BroadcastUpdatetoHUD)))

(fn DisableTimer
  (params (TrackedPlayer :type Controller%20Object%20Reference))
  (switch Utilities|FlowControl|Switch|SwitchonETimerDeviceState (CallFunction|GetTimerState TrackedPlayer)
    (:Enabled
      _)
    (:Activated
      _)
    (:Paused
      _)
    (:Completed
      _))
  (CallFunction|TimerSetState TrackedPlayer "Disabled"))

(fn FilterBotsFromFortPlayerStates
  (params (InFortPlayerStates :type Array%20of%20Fort%20Player%20State%20Object%20References))
  ;[0 1 0.0938 0.5] Returns a copy of the input array without bots.
  (for _array_element InFortPlayerStates
    (Utilities|IsValid _array_element
      (:Is%20Valid
        (if (not (Variables|/Script/Engine.PlayerState|GetbIsABot _array_element))
          (Utilities|Array|Add (LocalVariables|GetNonBotPlayersLocal) _array_element)))))
  (return (LocalVariables|GetNonBotPlayersLocal)))

(fn GetDuration
  (return (Utilities|GetCurrentDuration)))

(fn GetIntendedTrackedPlayer
  (params (Instigator :type Controller%20Object%20Reference))
  (if (CallFunction|CanClassInteract Instigator)
    (switch Utilities|FlowControl|Switch|SwitchonE_Timer_AppliesTo (Variables|UserOptions|GetAppliesTo)
      (:NewEnumerator0
        ;[1 1 1] Get individually tracked player
        (return (and (Utilities|Array|ContainsItem (CallFunction|GetTrackedPlayers) (Variables|Actor|GetInstigator)) (Utilities|IsValid (Variables|Actor|GetInstigator) :!class /Script/Engine.KismetSystemLibrary)) (Variables|Actor|GetInstigator)))
      (:NewEnumerator1
        (if (Creative|TeamOption|IsTeamType (Variables|UserOptions|GetActivatingTeam) "Any")
          _
          (elif (Math|Byte|Equal(Byte) (Team|GetActorTeam (Variables|Actor|GetInstigator)) (Variables|Default|GetDeviceTeamID))
            _
            (else
              ;[1 1 1] Instigator is not being tracked
              (return false 0))))
        ;[1 1 1] Global timer with null instigator
        (return true 0)))
    (else
      (return false 0))))

(fn GetIntendedUpdateDisplayPlayers
  (params (TrackedPlayer :type Controller%20Object%20Reference) (OutPlayersToUpdate :type Array%20of%20Controller%20Object%20References))
  (switch Utilities|FlowControl|Switch|SwitchonE_Timer_AppliesTo (Variables|UserOptions|GetAppliesTo)
    (:NewEnumerator0
      (Utilities|Array|Add (Variables|GetOutPlayersToUpdate) (Variables|GetTrackedPlayer)))
    (:NewEnumerator1
      (bind _currentminigame (Variables|/Script/FortniteGame.FortMinigameLogicComponent|GetCurrentMinigame (Variables|Components|GetFortMinigameProgress))
        (:then
          (if (Creative|TeamOption|IsTeamType (Variables|UserOptions|GetActivatingTeam) "Any")
            (LocalVariables|SetTargetPlayers (Minigame|GetParticipatingPlayers _currentminigame))
            (else
              (Utilities|IsValid (Minigame|GetMinigameTeamComponent _currentminigame)
                (:Is%20Valid
                  (LocalVariables|SetTargetPlayers (Class|FortMinigameTeamComponent|GetTeamMembers (Minigame|GetMinigameTeamComponent _currentminigame) (Variables|Default|GetDeviceTeamID)))))))
          (for _array_element (LocalVariables|GetTargetPlayers)
            (bind _ascontroller (Utilities|Casting|CastToController (Actor|GetOwner _array_element)))
            (Utilities|Array|Add (Variables|GetOutPlayersToUpdate) _ascontroller)))
        (:else)))))

(fn GetIsPaused
  (return (Utilities|Enum|Equal(Enum) (Variables|FortCreativeTimerDevice|GetClientCurrentState) "Paused")))

(fn GetIsStarted
  (return (Utilities|Enum|Equal(Enum) (Variables|FortCreativeTimerDevice|GetClientCurrentState) "Activated")))

(fn GetIsUrgently
  (return (Variables|Cosmetic|GetbUrgencyHasStarted)))

(fn GetLapTime
  (params (TrackedPlayer :type Controller%20Object%20Reference))
  (bind (_hasactivetimer _timeremaining) (CallFunction|GetRemainingTime TrackedPlayer))
  (return _hasactivetimer (Utilities|Select :Index (Variables|UserOptions|GetLapTimeStyle) :NewEnumerator0 (- (Utilities|GetCurrentDuration) _timeremaining) :NewEnumerator1 _timeremaining)))

(fn GetObjectToAutoSaveForPlayer
  (params (FortPlayerState :type Fort%20Player%20State%20Object%20Reference))
  (bind _returnvalue (Fortnite|GetControllerfromActor FortPlayerState))
  (Utilities|IsValid _returnvalue
    (:Is%20Valid
      (CallFunction|PopulatePersistenceDataObject _returnvalue)
      (return (Variables|Persistence|GetPersistenceData_Timer)))))

(fn GetPersistenceRequiredState
  (return (and (Utilities|Enum|Equal(Enum) (Variables|UserOptions|GetAppliesTo) "NewEnumerator0") (Variables|UserOptions|GetUsePersistence))))

(fn GetRemainingTime
  (params (TrackedPlayer :type Controller%20Object%20Reference))
  (return (> (CallFunction|GetTimerRemaining (Utilities|Select :Index (Variables|UserOptions|GetAppliesTo) :NewEnumerator0 TrackedPlayer)) 0) (CallFunction|GetTimerRemaining (Utilities|Select :Index (Variables|UserOptions|GetAppliesTo) :NewEnumerator0 TrackedPlayer))))

(fn GetTimeRemaining
  (params (Player :type Controller%20Object%20Reference))
  (bind (_hasactivetimer _timeremaining) (CallFunction|GetRemainingTime Player))
  (if _hasactivetimer
    (return _timeremaining)
    (else
      (return (Math|Float|SelectFloat (Utilities|GetCurrentDuration) 0.0 (Utilities|Enum|NotEqual(Enum) (CallFunction|GetTimerState Player) "Completed"))))))

(fn GetTimerName
  (return (Variables|UserOptions|GetTimerName)))

(fn GetTotalTime
  (bind _duration (Verse|GetDuration))
  (return (Math|Float|Truncate _duration)))

(fn GetWidgetClass
  (return (Variables|UserOptions|Advanced|GetCustomUserWidgetClass)))

(fn InitializeTimerText
  (Networking|CanExecuteCosmeticEvents
    (:True
      (Visual|UpdateTimerColor (Variables|UserOptions|GetTimerColor))
      (Visual|UpdateDisplayTime "Enabled" (Utilities|GetCurrentDuration)))))

(fn InitializeTrackedTimers
  (bind _fortminigameprogress (Variables|Components|GetFortMinigameProgress))
  (bind _currentminigame (Variables|/Script/FortniteGame.FortMinigameLogicComponent|GetCurrentMinigame _fortminigameprogress))
  (bind _returnvalue (Minigame|GetMinigameTeamComponent _currentminigame))
  (bind _fortminigameprogress_2 (Variables|Components|GetFortMinigameProgress))
  (bind _currentminigame_2 (Variables|/Script/FortniteGame.FortMinigameLogicComponent|GetCurrentMinigame _fortminigameprogress_2))
  (Networking|SwitchHasAuthority
    (:Authority
      (Utilities|FlowControl|Sequence
        (:then_0
          (CallFunction|ClearTrackedTimers)
          (Utilities|Map|Clear (Variables|Default|GetStateBeforeDisabledCache)))
        (:then_1
          (switch Utilities|FlowControl|Switch|SwitchonE_Timer_AppliesTo (Variables|UserOptions|GetAppliesTo)
            (:NewEnumerator0
              (if (Creative|TeamOption|IsTeamType (Variables|UserOptions|GetActivatingTeam) "Any")
                ;[1 1 1] Get every player and make a timer for each
                (Utilities|IsValid _currentminigame_2
                  (:Is%20Valid
                    (LocalVariables|SetApplicablePlayers (Minigame|GetParticipatingPlayers _currentminigame_2))))
                (else
                  ;[1 1 1] Get every player on the team and make a timer for each
                  (Utilities|IsValid _currentminigame
                    (:Is%20Valid
                      (Utilities|IsValid _returnvalue
                        (:Is%20Valid
                          (LocalVariables|SetApplicablePlayers (Class|FortMinigameTeamComponent|GetTeamMembers _returnvalue (Variables|Default|GetDeviceTeamID)))))))))
              (for _array_element (LocalVariables|GetApplicablePlayers)
                (bind _ascontroller (Utilities|Casting|CastToController (Actor|GetOwner _array_element)))
                (Default|StartTrackingPlayer _ascontroller)))
            (:NewEnumerator1
              ;[1 1 1] Use a single timer with no instigator
              (Default|StartTrackingPlayer 0))))))))

(fn IsFullscreen
  (return (Utilities|Enum|Equal(Enum) (Variables|UserOptions|Advanced|GetTimerAlignment) "NewEnumerator0")))

(fn IsStatePerPlayer
  (return (Utilities|Select :Index (Variables|UserOptions|GetAppliesTo) :NewEnumerator0 true :NewEnumerator1 false)))

(fn IsTargetableByLinearMover
  (return true))

(fn LoadTimer
  (params (PlayerController :type Controller%20Object%20Reference) (bUseTimeSinceSave :type Boolean))
  (bind _timesincesave (LocalVariables|GetTimeSinceSave))
  (bind _returnvalue (- (Variables|/Game/Creative/Devices/TimerDevice/Persistence/PersistenceData_Timer.PersistenceData_Timer_C|GetTimeRemaining (Variables|Persistence|GetPersistenceData_Timer)) _timesincesave))
  (bind _persistencedata_timer (Variables|Persistence|GetPersistenceData_Timer))
  (bind _persistencedata_timer_2 (Variables|Persistence|GetPersistenceData_Timer))
  (bind _timeremaining (Variables|/Game/Creative/Devices/TimerDevice/Persistence/PersistenceData_Timer.PersistenceData_Timer_C|GetTimeRemaining _persistencedata_timer_2))
  (Networking|SwitchHasAuthority
    (:Authority
      (if (and (Utilities|Enum|Equal(Enum) (Variables|UserOptions|GetAppliesTo) "NewEnumerator0") (Variables|UserOptions|GetUsePersistence))
        (if (CallFunction|CanClassInteract (Variables|GetPlayerController))
          (CallFunction|ResetPersistenceObject)
          (Persistence|LoadData (Variables|Default|GetFortCreativePersistenceComponentBase) (PlayerState|GetUniqueNetId (Variables|/Script/Engine.Controller|GetPlayerState (Variables|GetPlayerController))) (Variables|Persistence|GetSavePath) _persistencedata_timer)
          (if (Variables|/Game/Creative/Devices/TimerDevice/Persistence/PersistenceData_Timer.PersistenceData_Timer_C|GetbIsDataLoaded _persistencedata_timer)
            (GameplayEvent|BroadcastEvent (Variables|UserOptions|Events|GetOn%20Loaded) 0 (Utilities|Struct|MakeGameplayEventTimeronLoaded (Variables|GetPlayerController)))
            (Utilities|FlowControl|Sequence
              (:then_0
                ;[1 1 1] Upon new round the timer map is cleared, so we may need to recreate the tracked timer
                (Default|StartTrackingPlayer (Variables|GetPlayerController)))
              (:then_1
                ;[1 1 1] Reset timer time to saved value, and if necessary subtract out time since last save and get new time remaining
                (LocalVariables|SetTimeSinceSave (Math|Float|SelectFloat (Utilities|Select :Index (Variables|UserOptions|GetAuto-LoadWhen) :NewEnumerator0 (- (Fortnite|Sync|Time|GetServerWorldTimeSeconds) (Variables|/Game/Creative/Devices/TimerDevice/Persistence/PersistenceData_Timer.PersistenceData_Timer_C|GetServerTime (Variables|Persistence|GetPersistenceData_Timer))) :NewEnumerator1 (Math|Timespan|GetTotalSeconds (Math|DateTime|DateTime-DateTime (Math|DateTime|UTCNow) (Variables|/Game/Creative/Devices/TimerDevice/Persistence/PersistenceData_Timer.PersistenceData_Timer_C|GetSavedDateTime (Variables|Persistence|GetPersistenceData_Timer))))) 0.0 (Variables|GetbUseTimeSinceSave)))
                (LocalVariables|SetbTimerCompletedSinceSave (<= _returnvalue 0.0))
                (CallFunction|TimerStart (Variables|GetPlayerController) _returnvalue (- (+ (LocalVariables|GetTimeSinceSave) (- (Utilities|GetCurrentDuration) _timeremaining)) (- (Utilities|GetCurrentDuration) (Variables|UserOptions|GetUrgencyModeTime))))
                (CallFunction|TimerPause (Variables|GetPlayerController)))
              (:then_2
                ;[1 1 1] Save over the disabled cache if it was updated after saving a disabled timer, or clear it if a disabled timer is loaded to a non disabled state
                (if (and (Utilities|Enum|Equal(Enum) (Variables|/Game/Creative/Devices/TimerDevice/Persistence/PersistenceData_Timer.PersistenceData_Timer_C|GetState (Variables|Persistence|GetPersistenceData_Timer)) "Disabled") (or (Utilities|Enum|Equal(Enum) (Variables|/Game/Creative/Devices/TimerDevice/Persistence/PersistenceData_Timer.PersistenceData_Timer_C|GetStateBeforeDisabled (Variables|Persistence|GetPersistenceData_Timer)) "Activated") (Utilities|Enum|Equal(Enum) (Variables|/Game/Creative/Devices/TimerDevice/Persistence/PersistenceData_Timer.PersistenceData_Timer_C|GetStateBeforeDisabled (Variables|Persistence|GetPersistenceData_Timer)) "Paused")))
                  (Utilities|Map|Add (Variables|Default|GetStateBeforeDisabledCache) (Variables|GetPlayerController) (Variables|/Game/Creative/Devices/TimerDevice/Persistence/PersistenceData_Timer.PersistenceData_Timer_C|GetStateBeforeDisabled (Variables|Persistence|GetPersistenceData_Timer)))
                  (else
                    (Utilities|Map|Remove (Variables|Default|GetStateBeforeDisabledCache) (Variables|GetPlayerController)))))
              (:then_3
                (if (LocalVariables|GetbTimerCompletedSinceSave)
                  (CallFunction|TimerSetState (Variables|GetPlayerController) "Completed")
                  (else
                    (CallFunction|TimerSetState (Variables|GetPlayerController) (Variables|/Game/Creative/Devices/TimerDevice/Persistence/PersistenceData_Timer.PersistenceData_Timer_C|GetState (Variables|Persistence|GetPersistenceData_Timer)) true)))
                (return true)))
            (else
              (return false))))))))

(fn LoadTimerForAll
  (params (bUseTimeSinceSave :type Boolean))
  (if (Variables|UserOptions|GetUsePersistence)
    (bind _playerstate (Utilities|ForEachPersistentPlayer)
      (:LoopBody
        (Persistence|LoadTimer (Fortnite|GetFortPlayerControllerfromActor _playerstate) (Variables|GetbUseTimeSinceSave))))))

(fn MessageScore
  (return (Variables|UserOptions|Advanced|GetHUD%20Message) (Variables|UserOptions|Advanced|GetHUD%20MessageScoreColor) (Variables|UserOptions|Advanced|GetHUD%20MessageColor) (Variables|UserOptions|Advanced|GetResetHUD%20MessageScore)))

(fn MigrateLegacyUserOptions
  (bind _urgencymodetimeknob (Class|GetClassDefaults "/CreativeCoreDevices/Device_Timer_V2.Device_Timer_V2_C"))
  (bind _timertextstyleknob (Class|GetClassDefaults "/CreativeCoreDevices/Device_Timer_V2.Device_Timer_V2_C"))
  (bind _bdisplaytimeinminutessecondsknob (Class|GetClassDefaults "/CreativeCoreDevices/Device_Timer_V2.Device_Timer_V2_C"))
  (bind _bcountdowndirectionknob (Class|GetClassDefaults "/CreativeCoreDevices/Device_Timer_V2.Device_Timer_V2_C"))
  (bind _save_timer (Class|GetClassDefaults "/CreativeCoreDevices/Device_Timer_V2.Device_Timer_V2_C"))
  (bind _save_timer_2 (Variables|Migration|GetSave%20Timer))
  (bind _auto_save (Variables|UserOptions|GetAuto-Save))
  (|CallParentFunction)
  (Utilities|FlowControl|Sequence
    (:then_0
      (Utilities|FlowControl|Sequence
        (:then_0
          ;[0.2086 0.3763 1 0.25] Converting Urgency Mode options.
          (if (!= (Variables|Migration|GetUrgencyModeTimeKnob) _urgencymodetimeknob)
            (Variables|UserOptions|SetEnableUrgencyMode true)
            (Variables|UserOptions|SetUrgencyModeTime (Variables|Migration|GetUrgencyModeTimeKnob))
            (Variables|Migration|SetUrgencyModeTimeKnob _urgencymodetimeknob)
            (LocalVariables|SetOptionMigrated true)))
        (:then_1
          ;[0.2086 0.3763 1 0.25] Converting Text Style from Name to Enum
          (if (!= (Variables|Migration|GetTimerTextStyleKnob) _timertextstyleknob)
            (bind _keys (Utilities|Map|Keys (Variables|Default|GetTextStyleToStringMap)))
            (bind _array_element (Utilities|Array|ForEachLoopwithBreak _keys)
              (:LoopBody
                (if (== (Utilities|Map|Find (Variables|Default|GetTextStyleToStringMap) _array_element) (Variables|Migration|GetTimerTextStyleKnob))
                  (Variables|UserOptions|SetTimerLabelTextStyle _array_element)
                  (Variables|Migration|SetTimerTextStyleKnob _timertextstyleknob)
                  (LocalVariables|SetOptionMigrated true))))))
        (:then_2
          ;[0.2086 0.3763 1 0.25] Migrating Count Down Direction to Enum
          (if (!= (Variables|Migration|GetbCountDownDirectionKnob) _bcountdowndirectionknob)
            (Variables|UserOptions|Basic|SetCountDownDirection "NewEnumerator7")
            (Variables|Migration|SetbCountDownDirectionKnob _bcountdowndirectionknob)
            (LocalVariables|SetOptionMigrated true)))
        (:then_3
          ;[0.2086 0.3763 1 0.25] Migrating Display Time value to Enum
          (if (!= (Variables|Migration|GetbDisplayTimeInMinutesSecondsKnob) _bdisplaytimeinminutessecondsknob)
            (Variables|UserOptions|SetDisplayTimeIn "NewEnumerator7")
            (Variables|Migration|SetbDisplayTimeInMinutesSecondsKnob _bdisplaytimeinminutessecondsknob)
            (LocalVariables|SetOptionMigrated true)))
        (:then_4
          ;[0.2078 0.3765 1 0.251] Migrating Save Timer
          (if (Utilities|Enum|NotEqual(Enum) (Variables|Migration|GetSave%20Timer) _save_timer)
            (Variables|UserOptions|SetUsePersistence (Utilities|Select :Index _save_timer_2 :NewEnumerator0 false :NewEnumerator1 true :NewEnumerator7 true))
            (Variables|UserOptions|SetLoadElapsedTime (Utilities|Select :Index _save_timer_2 :NewEnumerator0 false :NewEnumerator1 false :NewEnumerator7 true))
            (Variables|Migration|SetSave%20Timer _save_timer)
            (LocalVariables|SetOptionMigrated true)))
        (:then_5
          ;[0.2078 0.3765 1 0.251] Migrating Auto-Save Auto-Load
          (if (Variables|Default|GetPersistenceAutoLoadAndSaveMigrated)
            _
            (elif (Variables|UserOptions|GetUsePersistence)
              (Variables|UserOptions|SetAuto-LoadWhen "NewEnumerator0")
              (if (!= (Class|GetClassDefaults "/CreativeCoreDevices/Device_Timer_V2.Device_Timer_V2_C") _auto_save)
                (Variables|UserOptions|SetAuto-SaveWhen (select _auto_save "NewEnumerator0" "NewEnumerator1")))
              (Variables|Default|SetPersistenceAutoLoadAndSaveMigrated true)
              (LocalVariables|SetOptionMigrated true)
              (else
                _))))
        (:then_6
          ;[0.2078 0.3765 1 0.251] Migrating Custom Widget Class
          (if (Utilities|IsValidClass (Variables|UserOptions|Advanced|GetCustomWidgetClass))
            (Variables|UserOptions|Advanced|SetCustomUserWidgetClass (Variables|UserOptions|Advanced|GetCustomWidgetClass))
            (Variables|UserOptions|Advanced|SetCustomWidgetClass 0)
            (Variables|UserOptions|Advanced|SetTimerAlignment "NewEnumerator1")
            (LocalVariables|SetOptionMigrated true)))))
    (:then_1
      ;[1 0 0.9385 0.25] If any option has migrated, refresh the saved options.
      (if (LocalVariables|GetOptionMigrated)
        (OptionData|RefreshPropertyOverrides (Variables|Components|GetToyOptionsComponent))))))

(fn OnClassApplied
  (params (PlayerState :type Fort%20Player%20State%20Object%20Reference) (ClassSlotIndex :type Byte))
  (bind _playerstate (Variables|GetPlayerState))
  (Networking|SwitchHasAuthority
    (:Authority
      (Utilities|IsValid _playerstate
        (:Is%20Valid
          (bind _ascontroller (Utilities|Casting|CastToController (Actor|GetOwner _playerstate)))
          (switch Utilities|FlowControl|Switch|SwitchonE_Timer_AppliesTo (Variables|UserOptions|GetAppliesTo)
            (:NewEnumerator0
              ;[1 1 1] Player leaving this device's class
              (CallFunction|CheckFailureDisableorReset _ascontroller))
            (:NewEnumerator1
              ;[1 1 1] Universal timer when ActivatingPlayer leaves this device's class
              (bind _activatingplayer (Variables|Default|GetActivatingPlayer)
                (:then
                  (if (== _activatingplayer _ascontroller)
                    (CallFunction|CheckFailureDisableorReset 0) ; We don't track player here because a unversal timer tracks its widgets with a null
                  ))
                (:else)))))))))

(fn OnLastChancetoSave
  (params (PlayerUniqueNetId :type Unique%20Net%20Id%20Structure) (bIsLocalPlayer :type Boolean))
  (Networking|SwitchHasAuthority
    (:Authority
      (if (Math|Boolean|ANDBoolean (Utilities|Enum|NotEqual(Enum) (Variables|UserOptions|GetAppliesTo) "NewEnumerator1") (Variables|UserOptions|GetUsePersistence) (Variables|UserOptions|GetAuto-Save))
        (Utilities|FlowControl|Sequence
          (:then_0
            (bind _playerstate (Utilities|ForEachPersistentPlayer)
              (:LoopBody
                (if (== (PlayerState|GetPlayerUniqueNetID _playerstate) (Variables|GetPlayerUniqueNetId))
                  (LocalVariables|SetLocalPlayerState _playerstate)))))
          (:then_1
            (bind _localplayerstate (LocalVariables|GetLocalPlayerState)
              (:then
                (Utilities|IsValid (CreativeAI|Controller|GetParticipantControllerfromActor _localplayerstate)
                  (:Is%20Valid
                    (Persistence|SaveTimers (CreativeAI|Controller|GetParticipantControllerfromActor _localplayerstate))
                    (CallFunction|TimerSetState (CreativeAI|Controller|GetParticipantControllerfromActor _localplayerstate) "Paused")
                    (CallFunction|ServerRequestClearDisplay (Utilities|Array|MakeArray (CreativeAI|Controller|GetParticipantControllerfromActor _localplayerstate))))))
              (:else
                ;[1 1 1] Logs
                (bind _returnvalue (PlayerInfo|GetDebugStringforUniqueId (Variables|GetPlayerUniqueNetId)))
                (Fortnite|FortShippingLog self (Utilities|String|Append "FortCreativePersistenceComponentBase - OnLastChanceToSave: Do not exists a valid Player State for the UniqueNetID:" _returnvalue) true)))))))))

(fn OnClientDisplayUpdateTimerCompleted
  (Visual|UpdateinGameDisplay "Activated" (- (Variables|Cosmetic|GetClientDisplayUpdateEndTime) (Utilities|Time|GetTimeSeconds))))

(fn OnClientExitVolume
  (params (Client :type Player%20State%20Object%20Reference) (Volume :type Actor%20Object%20Reference))
  (Networking|SwitchHasAuthority
    (:Authority
      ;[1 1 1] Check that we're exiting our specific island, and that this is not a global timer
      (if (Math|Boolean|ANDBoolean (Utilities|IsValid (Minigame|GetMinigame Volume :!class /Script/FortniteGame.MinigameBlueprintLibrary) :!class /Script/Engine.KismetSystemLibrary) (Utilities|IsValid (Variables|/Script/FortniteGame.FortMinigameLogicComponent|GetCurrentMinigame (Variables|Components|GetFortMinigameProgress)) :!class /Script/Engine.KismetSystemLibrary) (Utilities|Equal(Object) (Minigame|GetMinigame Volume :!class /Script/FortniteGame.MinigameBlueprintLibrary) (Variables|/Script/FortniteGame.FortMinigameLogicComponent|GetCurrentMinigame (Variables|Components|GetFortMinigameProgress))) (Utilities|Enum|NotEqual(Enum) (Variables|UserOptions|GetAppliesTo) "NewEnumerator1"))
        (bind _ascontroller (Utilities|Casting|CastToController (Actor|GetOwner (Variables|GetClient))))
        (if (not (Math|Boolean|ANDBoolean (Variables|UserOptions|GetAuto-Save) (Variables|UserOptions|GetUsePersistence) (Utilities|Enum|Equal(Enum) (Variables|UserOptions|GetAuto-SaveWhen) "NewEnumerator0")))
          (Default|StopTrackingPlayer _ascontroller))))))

(fn OnMinigameStateChanged
  (params (Minigame :type Fort%20Minigame%20Object%20Reference) (NewMinigameState :type EFortMinigameState%20Enum :default "PreGame"))
  (switch Utilities|FlowControl|Switch|SwitchonEFortMinigameState NewMinigameState
    (:PreGame
      _)
    (:Setup
      _)
    (:PostRoundEnd
      _))
  (Networking|CanExecuteCosmeticEvents
    (:True
      (Visual|ClearDisplayTexts 0 false))))

(fn OnPlayerJoinInProgress
  (params (Pawn :type Pawn%20Object%20Reference))
  (bind _pawn (Variables|GetPawn))
  (bind _returnvalue (Pawn|GetController _pawn))
  (Networking|SwitchHasAuthority
    (:Authority
      (if (Utilities|Enum|NotEqual(Enum) (Variables|UserOptions|GetAppliesTo) "NewEnumerator1")
        (if (Creative|TeamOption|IsTeamType (Variables|UserOptions|GetActivatingTeam) "Any")
          _
          (elif (Math|Byte|Equal(Byte) (Variables|Default|GetDeviceTeamID) (Team|GetActorTeam (Variables|GetPawn)))
            _))
        (if (and (Variables|UserOptions|GetAuto-Load) (or (Variables|UserOptions|GetAuto-Save) (Utilities|Enum|Equal(Enum) (Variables|UserOptions|GetAuto-LoadWhen) "NewEnumerator1")))
          (bind _bloadeddata (Persistence|LoadTimer _returnvalue (and (Variables|UserOptions|GetLoadElapsedTime) (Variables|UserOptions|GetUsePersistence))))
          (if _bloadeddata
            _
            (else
              _))
          (else
            _))
        (Default|StopTrackingPlayer _returnvalue) ; If Auto-Save enabled but Auto-Load is not, need to discard existing paused timer before starting to track again
        (Default|StartTrackingPlayer _returnvalue)
        (else
          ;[1 1 1] This is a timer for everyone, simply update the timer display for the joining PC
          (CallFunction|ServerRequestDisplayUpdate 0 (Utilities|Array|MakeArray (Pawn|GetController (Variables|GetPawn)))))))))

(fn OnPlayerRemovedFromMinigame
  (params (Unique%20Net%20Id :type Unique%20Net%20Id%20Structure) (bIsLocalPlayer :type Boolean))
  (bind _lastinstigator (Variables|Default|GetLastInstigator))
  (Networking|SwitchHasAuthority
    (:Authority
      (Utilities|IsValid _lastinstigator
        (:Is%20Valid
          (bind _asfort_player_state (Utilities|Casting|CastToFortPlayerState (Variables|/Script/Engine.Controller|GetPlayerState _lastinstigator)))
          (if (== (PlayerState|GetPlayerUniqueNetID _asfort_player_state) Unique%20Net%20Id)
            (Variables|Default|SetbInstigatorWasRemoved true)
            (if (Utilities|Enum|Equal(Enum) (Variables|UserOptions|GetIfInstigatingPlayerIsNotPresent) "NewEnumerator0")
              ;[1 1 1] Clear Last Instigator
              (Variables|Default|SetLastInstigator 0)
              (else
                (bind _playercontroller (Utilities|GetRandomTeamMember)
                  (:HasValidMember
                    (Variables|Default|SetLastInstigator _playercontroller))
                  (:NoValidMember
                    (Variables|Default|SetLastInstigator 0)))))))))))

(fn OnRep_DurationOverride
  (Networking|CanExecuteCosmeticEvents
    (:True
      (if (Utilities|Enum|Equal(Enum) (Variables|FortCreativeTimerDevice|GetClientCurrentState) "Enabled")
        (CallFunction|InitializeTimerText)))))

(fn OnServerDisplayUpdateTimerCompleted
  (params (TrackedPlayer :type Controller%20Object%20Reference))
  (CallFunction|RequestServerUpdateForPlayer TrackedPlayer))

(fn OnTeamChanged
  (params (PlayerState :type Fort%20Player%20State%20Object%20Reference) (TeamId :type Byte) (OldTeamId :type Byte) (bTeamChangedWithoutRespawn :type Boolean))
  (bind _playerstate (Variables|GetPlayerState))
  (bind _playerstate_2 (Variables|GetPlayerState))
  (bind _playerstate_3 (Variables|GetPlayerState))
  (bind _playerstate_4 (Variables|GetPlayerState))
  (bind _playerstate_5 (Variables|GetPlayerState))
  (Networking|SwitchHasAuthority
    (:Authority
      ;[1 1 1] Check that this is a team timer
      (if (not (Creative|TeamOption|IsTeamType (Variables|UserOptions|GetActivatingTeam) "Any"))
        (Utilities|FlowControl|Sequence
          (:then_0
            ;[1 1 1] Handle personal timers
            (switch Utilities|FlowControl|Switch|SwitchonE_Timer_AppliesTo (Variables|UserOptions|GetAppliesTo)
              (:NewEnumerator0
                ;;[1 1 1] Player joining this device's team
                (if (Math|Byte|Equal(Byte) (Variables|GetTeamId) (Variables|Default|GetDeviceTeamID))
                  (Utilities|IsValid _playerstate_5
                    (:Is%20Valid
                      (bind _ascontroller (Utilities|Casting|CastToController (Actor|GetOwner _playerstate_5)))
                      (Default|StartTrackingPlayer _ascontroller)))
                  (elif (Math|Byte|Equal(Byte) (Variables|GetOldTeamId) (Variables|Default|GetDeviceTeamID))
                    ;;[1 1 1] Player leaving this device's team
                    (Utilities|IsValid _playerstate_2
                      (:Is%20Valid
                        (bind _ascontroller_2 (Utilities|Casting|CastToController (Actor|GetOwner _playerstate_2)))
                        (CallFunction|CheckFailureDisableorReset _ascontroller_2)
                        (Default|StopTrackingPlayer _ascontroller_2))))))
              (:NewEnumerator1
                ;[1 1 1] Handle universal timer
                ;;[1 1 1] Player joining this device's team
                (if (Math|Byte|Equal(Byte) (Variables|GetTeamId) (Variables|Default|GetDeviceTeamID))
                  (Utilities|IsValid _playerstate
                    (:Is%20Valid
                      (bind _ascontroller_3 (Utilities|Casting|CastToController (Actor|GetOwner _playerstate)))
                      (CallFunction|ServerRequestDisplayUpdate 0 (Utilities|Array|MakeArray _ascontroller_3))))
                  (elif (Math|Byte|Equal(Byte) (Variables|GetOldTeamId) (Variables|Default|GetDeviceTeamID))
                    ;;[1 1 1] Universal timer when Activating Player leaves this device's team
                    (Utilities|IsValid _playerstate_3
                      (:Is%20Valid
                        (bind _ascontroller_4 (Utilities|Casting|CastToController (Actor|GetOwner _playerstate_3)))
                        (bind _activatingplayer (Variables|Default|GetActivatingPlayer)
                          (:then
                            (if (== _activatingplayer _ascontroller_4)
                              (CallFunction|CheckFailureDisableorReset 0) ; We don't track player here because a unversal timer tracks its widgets with a null
                            ))
                          (:else)))))))))
          (:then_1
            ;[1 1 1] Client joining some other team, need to erase text
            (if (Math|Byte|NotEqual(Byte) (Variables|GetTeamId) (Variables|Default|GetDeviceTeamID))
              (Utilities|IsValid _playerstate_4
                (:Is%20Valid
                  (bind _ascontroller_5 (Utilities|Casting|CastToController (Actor|GetOwner _playerstate_4)))
                  (CallFunction|ServerRequestClearDisplay (Utilities|Array|MakeArray _ascontroller_5)))))))))))

(fn OnTimerCompleted
  (params (TrackedPlayer :type Controller%20Object%20Reference))
  (CallFunction|TimerSetState TrackedPlayer "Completed"))

(fn OnTimerStateChanged
  (params (TrackedPlayer :type Controller%20Object%20Reference) (TimerHandle :type Timer%20Handle%20Structure%20%28by%20ref%29) (CurrentState :type ETimerDeviceState%20Enum :default "Enabled") (PreviousState :type ETimerDeviceState%20Enum :default "Enabled") (bLoadingFromSave :type Boolean))
  (bind _trackedplayer (Variables|GetTrackedPlayer))
  (bind _statebeforedisabledcache (Variables|Default|GetStateBeforeDisabledCache))
  (bind (_value _returnvalue) (Utilities|Map|Find _statebeforedisabledcache _trackedplayer))
  (bind _fortminigameprogress (Variables|Components|GetFortMinigameProgress))
  (bind _currentminigame (Variables|/Script/FortniteGame.FortMinigameLogicComponent|GetCurrentMinigame _fortminigameprogress))
  (bind _returnvalue_2 (Minigame|GetMinigameTeamComponent _currentminigame))
  (bind _timerhandle (Variables|GetTimerHandle))
  (bind _fortminigameprogress_2 (Variables|Components|GetFortMinigameProgress))
  (bind _currentminigame_2 (Variables|/Script/FortniteGame.FortMinigameLogicComponent|GetCurrentMinigame _fortminigameprogress_2))
  (bind _oncompleteinstigator (LocalVariables|GetOnCompleteInstigator))
  (Utilities|FlowControl|Sequence
    (:then_0
      (switch Utilities|FlowControl|Switch|SwitchonETimerDeviceState (Variables|GetCurrentState)
        (:Enabled
          ;[0 0 1] Enabled
          (if (Math|Boolean|ANDBoolean (not (Variables|GetbLoadingFromSave)) (Utilities|Enum|Equal(Enum) (Variables|GetPreviousState) "Disabled") _returnvalue)
            (CallFunction|TimerSetState (Variables|GetTrackedPlayer) _value)
            (Utilities|Map|Remove (Variables|Default|GetStateBeforeDisabledCache) (Variables|GetTrackedPlayer))
            (else
              (CallFunction|TimerStart (Variables|GetTrackedPlayer) (Utilities|GetCurrentDuration) (- (Utilities|GetCurrentDuration) (Variables|UserOptions|GetUrgencyModeTime)))
              (CallFunction|TimerPause (Variables|GetTrackedPlayer)))))
        (:Activated
          ;[0 0.995 0] Activated
          (if (Utilities|Time|IsTimerPausedbyHandle _timerhandle)
            (CallFunction|TimerResume (Variables|GetTrackedPlayer))
            (elif (not (Utilities|Time|IsTimerActivebyHandle _timerhandle))
              (CallFunction|TimerStart (Variables|GetTrackedPlayer) (Utilities|GetCurrentDuration) (- (Utilities|GetCurrentDuration) (Variables|UserOptions|GetUrgencyModeTime))))))
        (:Paused
          ;[1 1 0] Paused
          (CallFunction|TimerPause (Variables|GetTrackedPlayer)))
        (:Completed
          ;[0 1 1] Completed
          (Utilities|FlowControl|Sequence
            (:then_0
              ;;[1 1 1] Handle Award Score Behavior & Lap Time
              (if (not (Variables|GetbLoadingFromSave))
                (switch Utilities|FlowControl|Switch|SwitchonE_Timer_AppliesTo (Variables|UserOptions|GetAppliesTo)
                  (:NewEnumerator0
                    (bind _asfort_player_controller (Utilities|Casting|CastToFortPlayerController (Variables|GetTrackedPlayer)))
                    (Default|AwardStats _asfort_player_controller (Utilities|Time|GetTimerRemainingTimebyHandle (Variables|GetTimerHandle))))
                  (:NewEnumerator1
                    (if (Creative|TeamOption|IsTeamType (Variables|UserOptions|GetActivatingTeam) "Any")
                      (Utilities|IsValid _currentminigame_2
                        (:Is%20Valid
                          (LocalVariables|SetApplicable%20Players (Minigame|GetParticipatingPlayers _currentminigame_2))))
                      (else
                        (Utilities|IsValid _currentminigame
                          (:Is%20Valid
                            (Utilities|IsValid _returnvalue_2
                              (:Is%20Valid
                                (LocalVariables|SetApplicable%20Players (Class|FortMinigameTeamComponent|GetTeamMembers _returnvalue_2 (Variables|Default|GetDeviceTeamID)))))))))
                    (for _array_element (LocalVariables|GetApplicable%20Players)
                      (bind _asfort_player_controller_2 (Utilities|Casting|CastToFortPlayerController (Actor|GetOwner _array_element)))
                      (if (StateChecks|IsValidClass (Variables|Default|GetCreative_UseRestriction_Component) _asfort_player_controller_2)
                        (Default|AwardStats _asfort_player_controller_2 (Utilities|Time|GetTimerRemainingTimebyHandle (Variables|GetTimerHandle)))))))))
            (:then_1
              ;;[1 1 1] Handle Channel Trigger Behavior
              (if (Math|Boolean|ANDBoolean (not (Variables|GetbLoadingFromSave)) (Utilities|Enum|Equal(Enum) (Variables|UserOptions|GetAppliesTo) "NewEnumerator1") (Creative|TeamOption|IsTeamType (Variables|UserOptions|GetActivatingTeam) "Any"))
                (bind _lastinstigator (Variables|Default|GetLastInstigator)
                  (:then
                    (LocalVariables|SetOnCompleteInstigator _lastinstigator))
                  (:else
                    (if (and (not (Variables|Default|GetbInstigatorWasRemoved)) (Utilities|Enum|Equal(Enum) (Variables|UserOptions|GetIfInstigatingPlayerIsNotPresent) "NewEnumerator7"))
                      (bind _playercontroller (Utilities|GetRandomTeamMember)
                        (:HasValidMember
                          (LocalVariables|SetOnCompleteInstigator _playercontroller))
                        (:NoValidMember
                          (LocalVariables|SetOnCompleteInstigator (Variables|GetTrackedPlayer))))
                      (else
                        (LocalVariables|SetOnCompleteInstigator 0)))))
                (Variables|Default|SetbInstigatorWasRemoved false)
                (if (CallFunction|TimerWasSuccessful (Utilities|Time|GetTimerRemainingTimebyHandle (Variables|GetTimerHandle)))
                  (GameplayEvent|BroadcastEvent (Variables|UserOptions|Events|GetOnSuccess) 0 (Utilities|Struct|MakeGameplayEventTimerSuccess _oncompleteinstigator))
                  (else
                    (GameplayEvent|BroadcastEvent (Variables|UserOptions|Events|GetOnFailure) 0 (Utilities|Struct|MakeGameplayEventTimerFailure _oncompleteinstigator))))
                (else
                  _)))
            (:then_2
              ;;[1 1 1] Handle On Complete Behavior
              (CallFunction|TimerPause (Variables|GetTrackedPlayer))
              (switch Utilities|FlowControl|Switch|SwitchonE_Timer_OnCompleteBehavior (Variables|UserOptions|GetCompletionBehavior)
                (:NewEnumerator0
                  (LocalVariables|SetOnCompleteNewState "Disabled"))
                (:NewEnumerator2
                  (LocalVariables|SetOnCompleteNewState "Enabled"))
                (:NewEnumerator3
                  (LocalVariables|SetOnCompleteNewState "Activated")))
              (LocalVariables|SetbOnCompleteChangeState true))))
        (:Disabled
          ;[1 0 0] Disabled
          (CallFunction|TimerPause (Variables|GetTrackedPlayer))
          (if (and (not (Variables|GetbLoadingFromSave)) (or (Utilities|Enum|Equal(Enum) (Variables|GetPreviousState) "Activated") (Utilities|Enum|Equal(Enum) (Variables|GetPreviousState) "Paused")))
            (Utilities|Map|Add (Variables|Default|GetStateBeforeDisabledCache) (Variables|GetTrackedPlayer) (Variables|GetPreviousState))))))
    (:then_1
      (Default|GetIntendedUpdateDisplayPlayers (Variables|GetTrackedPlayer) (LocalVariables|GetPlayersToUpdate))
      (CallFunction|ServerRequestDisplayUpdate (Variables|GetTrackedPlayer) (LocalVariables|GetPlayersToUpdate)))
    (:then_2
      (if (LocalVariables|GetbOnCompleteChangeState)
        (CallFunction|CompletionStateChangeDelay (Variables|GetTrackedPlayer) (LocalVariables|GetOnCompleteNewState))))))

(fn OnUrgencyTimeReached
  (params (TrackedPlayer :type Controller%20Object%20Reference))
  (Default|GetIntendedUpdateDisplayPlayers TrackedPlayer (LocalVariables|GetPlayersToUpdate))
  (CallFunction|ServerRequestDisplayUpdate (Variables|GetTrackedPlayer) (LocalVariables|GetPlayersToUpdate))
  (GameplayEvent|BroadcastEvent (Variables|UserOptions|Events|GetOnStartUrgencyMode) 0 (Utilities|Struct|MakeGameplayEventTimerUrgency (Variables|GetTrackedPlayer))))

(fn PopulatePersistenceDataObject
  (params (TrackedPlayer :type Controller%20Object%20Reference))
  (bind (_outstate _returnvalue) (CallFunction|GetTimerState TrackedPlayer))
  (bind (_value _returnvalue_2) (Utilities|Map|Find (Variables|Default|GetStateBeforeDisabledCache) (Variables|GetTrackedPlayer)))
  (bind _persistencedata_timer (Variables|Persistence|GetPersistenceData_Timer))
  (CallFunction|ResetPersistenceObject)
  (if _returnvalue
    (bind _output_get (Variables|/Game/Creative/Devices/TimerDevice/Persistence/PersistenceData_Timer.PersistenceData_Timer_C|SetState _outstate _persistencedata_timer))
    (Variables|/Game/Creative/Devices/TimerDevice/Persistence/PersistenceData_Timer.PersistenceData_Timer_C|SetStateBeforeDisabled (select (and _returnvalue_2 (Utilities|Enum|Equal(Enum) _output_get "Disabled")) _value "Enabled") _persistencedata_timer)
    (Variables|/Game/Creative/Devices/TimerDevice/Persistence/PersistenceData_Timer.PersistenceData_Timer_C|SetTimeRemaining (CallFunction|GetTimerRemaining (Variables|GetTrackedPlayer)) _persistencedata_timer)
    (Variables|/Game/Creative/Devices/TimerDevice/Persistence/PersistenceData_Timer.PersistenceData_Timer_C|SetServerTime (Fortnite|Sync|Time|GetServerWorldTimeSeconds) _persistencedata_timer)
    (Variables|/Game/Creative/Devices/TimerDevice/Persistence/PersistenceData_Timer.PersistenceData_Timer_C|SetSavedDateTime (Math|DateTime|UTCNow) _persistencedata_timer)
    (Variables|/Game/Creative/Devices/TimerDevice/Persistence/PersistenceData_Timer.PersistenceData_Timer_C|SetbIsDataLoaded true _persistencedata_timer)))

(fn RequestServerUpdateforPlayer
  (params (Player :type Controller%20Object%20Reference))
  (Default|GetIntendedUpdateDisplayPlayers Player (LocalVariables|GetPlayersToUpdate))
  (CallFunction|ServerRequestDisplayUpdate Player (LocalVariables|GetPlayersToUpdate)))

(fn ResetPersistenceObject
  (bind _persistencedata_timer (Variables|Persistence|GetPersistenceData_Timer))
  (Variables|/Game/Creative/Devices/TimerDevice/Persistence/PersistenceData_Timer.PersistenceData_Timer_C|SetState "Enabled" _persistencedata_timer)
  (Variables|/Game/Creative/Devices/TimerDevice/Persistence/PersistenceData_Timer.PersistenceData_Timer_C|SetStateBeforeDisabled "Enabled" _persistencedata_timer)
  (Variables|/Game/Creative/Devices/TimerDevice/Persistence/PersistenceData_Timer.PersistenceData_Timer_C|SetTimeRemaining 0.0 _persistencedata_timer)
  (Variables|/Game/Creative/Devices/TimerDevice/Persistence/PersistenceData_Timer.PersistenceData_Timer_C|SetServerTime 0.0 _persistencedata_timer)
  (Variables|/Game/Creative/Devices/TimerDevice/Persistence/PersistenceData_Timer.PersistenceData_Timer_C|SetSavedDateTime :self _persistencedata_timer)
  (Variables|/Game/Creative/Devices/TimerDevice/Persistence/PersistenceData_Timer.PersistenceData_Timer_C|SetbIsDataLoaded false _persistencedata_timer))

(fn ResetTimerForAll
  (params (Instigator :type Controller%20Object%20Reference))
  (switch Utilities|FlowControl|Switch|SwitchonE_Timer_AppliesTo (Variables|UserOptions|GetAppliesTo)
    (:NewEnumerator0
      (bind _bresult (Default|CanTriggerAllPersonalTimers Instigator))
      (if _bresult
        (for _array_element (CallFunction|GetTrackedPlayers)
          (CallFunction|ResetTimer _array_element))))
    (:NewEnumerator1
      (CallFunction|ResetTimer 0))))

(fn ResetTimer
  (params (TrackedPlayer :type Controller%20Object%20Reference))
  (switch Utilities|FlowControl|Switch|SwitchonETimerDeviceState (CallFunction|GetTimerState TrackedPlayer)
    (:Activated
      _)
    (:Paused
      _)
    (:Completed
      _))
  (CallFunction|TimerClearHandles TrackedPlayer)
  (CallFunction|TimerSetState TrackedPlayer))

(fn SaveTimers
  (params (PlayerController :type Controller%20Object%20Reference))
  (bind (_outstate _returnvalue) (CallFunction|GetTimerState (Variables|GetPlayerController)))
  (bind (_value _returnvalue_2) (Utilities|Map|Find (Variables|Default|GetStateBeforeDisabledCache) (Variables|GetPlayerController)))
  (bind _persistencedata_timer (Variables|Persistence|GetPersistenceData_Timer))
  (bind _playercontroller (Variables|GetPlayerController))
  (Networking|SwitchHasAuthority
    (:Authority
      (CallFunction|ResetPersistenceObject)
      (if (Math|Boolean|ANDBoolean (Utilities|Enum|Equal(Enum) (Variables|UserOptions|GetAppliesTo) "NewEnumerator0") (Variables|UserOptions|GetUsePersistence) (or (Creative|TeamOption|IsTeamType (Variables|UserOptions|GetActivatingTeam) "Any") (Math|Byte|Equal(Byte) (Variables|Default|GetDeviceTeamID) (Team|GetActorTeam _playercontroller))) (CallFunction|CanClassInteract _playercontroller) _returnvalue)
        (bind _output_get (Variables|/Game/Creative/Devices/TimerDevice/Persistence/PersistenceData_Timer.PersistenceData_Timer_C|SetState _outstate _persistencedata_timer))
        (Variables|/Game/Creative/Devices/TimerDevice/Persistence/PersistenceData_Timer.PersistenceData_Timer_C|SetStateBeforeDisabled (select (and _returnvalue_2 (Utilities|Enum|Equal(Enum) _output_get "Disabled")) _value "Enabled") _persistencedata_timer)
        (Variables|/Game/Creative/Devices/TimerDevice/Persistence/PersistenceData_Timer.PersistenceData_Timer_C|SetTimeRemaining (CallFunction|GetTimerRemaining (Variables|GetPlayerController)) _persistencedata_timer)
        (Variables|/Game/Creative/Devices/TimerDevice/Persistence/PersistenceData_Timer.PersistenceData_Timer_C|SetServerTime (Fortnite|Sync|Time|GetServerWorldTimeSeconds) _persistencedata_timer)
        (Variables|/Game/Creative/Devices/TimerDevice/Persistence/PersistenceData_Timer.PersistenceData_Timer_C|SetSavedDateTime (Math|DateTime|UTCNow) _persistencedata_timer)
        (Variables|/Game/Creative/Devices/TimerDevice/Persistence/PersistenceData_Timer.PersistenceData_Timer_C|SetbIsDataLoaded true _persistencedata_timer)
        (Persistence|SaveData (Variables|Default|GetFortCreativePersistenceComponentBase) _persistencedata_timer (PlayerState|GetUniqueNetId (Variables|/Script/Engine.Controller|GetPlayerState (Variables|GetPlayerController))) (Variables|Persistence|GetSavePath) false)
        (GameplayEvent|BroadcastEvent (Variables|UserOptions|Events|GetOn%20Saved) 0 (Utilities|Struct|MakeGameplayEventTimeronSaved (Variables|GetPlayerController)))))))

(fn SetLapTimeforPlayer
  (params (Player :type Controller%20Object%20Reference))
  (bind _checkpc (LocalVariables|GetCheckPC))
  (Networking|SwitchHasAuthority
    (:Authority
      (bind (_hasactivetimer _returnvalue) (CallFunction|GetLapTime (Variables|GetPlayer)))
      (LocalVariables|SetLapTime _returnvalue)
      (bind _player (Variables|GetPlayer)
        (:then
          (bind _asfort_player_controller (Utilities|Casting|CastToFortPlayerController _player))
          (bind _currentminigame (Variables|/Script/FortniteGame.FortMinigameLogicComponent|GetCurrentMinigame (Variables|Components|GetFortMinigameProgress))
            (:then
              (Minigame|SetLapTimeStat (Minigame|Stats|GetMinigameStatsComponent _currentminigame) _asfort_player_controller (LocalVariables|GetLapTime)))
            (:else)))
        (:else
          (if (Creative|TeamOption|IsTeamType (Variables|UserOptions|GetActivatingTeam) "Any")
            (bind _currentminigame_2 (Variables|/Script/FortniteGame.FortMinigameLogicComponent|GetCurrentMinigame (Variables|Components|GetFortMinigameProgress))
              (:then
                (LocalVariables|SetApplicable%20Players (Minigame|GetParticipatingPlayers _currentminigame_2)))
              (:else))
            (else
              (bind _currentminigame_3 (Variables|/Script/FortniteGame.FortMinigameLogicComponent|GetCurrentMinigame (Variables|Components|GetFortMinigameProgress))
                (:then
                  (Utilities|IsValid (Minigame|GetMinigameTeamComponent _currentminigame_3)
                    (:Is%20Valid
                      (LocalVariables|SetApplicable%20Players (Class|FortMinigameTeamComponent|GetTeamMembers (Minigame|GetMinigameTeamComponent _currentminigame_3) (Variables|Default|GetDeviceTeamID))))))
                (:else))))
          (for _array_element (LocalVariables|GetApplicable%20Players)
            (bind _asfort_player_controller_2 (Utilities|Casting|CastToFortPlayerController (Actor|GetOwner _array_element)))
            (LocalVariables|SetCheckPC _asfort_player_controller_2)
            (if (StateChecks|IsValidClass (Variables|Default|GetCreative_UseRestriction_Component) _asfort_player_controller_2)
              (switch Utilities|FlowControl|Switch|SwitchonE_Timer_AppliesTo (Variables|UserOptions|GetAppliesTo)
                (:NewEnumerator0
                  (bind (_hasactivetimer_2 _returnvalue_2) (CallFunction|GetLapTime _checkpc))
                  (if _hasactivetimer_2
                    (Minigame|SetLapTimeStat (Minigame|Stats|GetMinigameStatsComponent (Variables|/Script/FortniteGame.FortMinigameLogicComponent|GetCurrentMinigame (Variables|Components|GetFortMinigameProgress))) _checkpc _returnvalue_2)))
                (:NewEnumerator1
                  (Minigame|SetLapTimeStat (Minigame|Stats|GetMinigameStatsComponent (Variables|/Script/FortniteGame.FortMinigameLogicComponent|GetCurrentMinigame (Variables|Components|GetFortMinigameProgress))) (LocalVariables|GetCheckPC) (LocalVariables|GetLapTime)))))))))))

(fn SetDuration
  (params (NewDuration :type Float%20%28double-precision%29))
  (bind _returnvalue (Math|Float|Clamp(Float) NewDuration 1.0 3.6e+03))
  (Networking|SwitchHasAuthority
    (:Authority
      (LocalVariables|SetDifference (- _returnvalue (Utilities|GetCurrentDuration)))
      (Variables|Verse|SetDurationOverride _returnvalue)
      (switch Utilities|FlowControl|Switch|SwitchonE_Timer_AppliesTo (Variables|UserOptions|GetAppliesTo)
        (:NewEnumerator0
          (for _array_element (CallFunction|GetTrackedPlayers)
            (Verse|AddTime _array_element (LocalVariables|GetDifference))))
        (:NewEnumerator1
          (Verse|AddTime 0 (LocalVariables|GetDifference)))))))

(fn SetTimeRemaining
  (params (Player :type Controller%20Object%20Reference) (TimeValue :type Float%20%28double-precision%29))
  (Networking|SwitchHasAuthority
    (:Authority
      (bind _hasactivetimer (CallFunction|GetRemainingTime Player))
      (if _hasactivetimer
        (CallFunction|TimerStart Player (Math|Float|Clamp(Float) TimeValue 1.0 (Utilities|GetCurrentDuration)) (- (Math|Float|Clamp(Float) TimeValue 1.0 (Utilities|GetCurrentDuration)) (Variables|UserOptions|GetUrgencyModeTime)))
        (CallFunction|RequestServerUpdateForPlayer Player)))))

(fn SetupClientDisplayUpdateTimer
  (params (CurrentState :type ETimerDeviceState%20Enum :default "Enabled") (TimeRemaining :type Float%20%28double-precision%29))
  (if (Utilities|Enum|Equal(Enum) CurrentState "Activated")
    (Variables|Cosmetic|SetClientDisplayUpdateEndTime (+ (Utilities|Time|GetTimeSeconds) (Variables|GetTimeRemaining)))
    (bind _returnvalue (Utilities|Time|SetTimerbyEvent (EventDispatchers|CreateEvent :!event OnClientDisplayUpdateTimerCompleted) (Variables|Cosmetic|GetClientDisplayUpdateRate) true false 0.0 0.3))
    (Variables|Cosmetic|SetClientDisplayUpdateTimerHandle _returnvalue)
    (else
      (Utilities|Time|ClearandInvalidateTimerbyHandle (Variables|Cosmetic|GetClientDisplayUpdateTimerHandle)))))

(fn ShouldInvisiblePropsBeAllowed
  (return false))

(fn StartTrackingPlayer
  (params (PlayerToTrack :type Controller%20Object%20Reference))
  (if (not (Utilities|Array|ContainsItem (CallFunction|GetTrackedPlayers) (Variables|GetPlayerToTrack)))
    (CallFunction|CreateTrackedTimer (Variables|GetPlayerToTrack) (select (Variables|UserOptions|Basic|GetStartAtGameStart) (Utilities|Enum|LiteralenumETimerDeviceState "Activated") (Utilities|Enum|LiteralenumETimerDeviceState "Enabled")))))

(fn StopTrackingPlayer
  (params (TrackedPlayer :type Controller%20Object%20Reference))
  (if (Utilities|Array|ContainsItem (CallFunction|GetTrackedPlayers) TrackedPlayer)
    (Utilities|Map|Remove (Variables|Default|GetStateBeforeDisabledCache) TrackedPlayer)
    (CallFunction|TimerClearHandles TrackedPlayer)
    (CallFunction|RemoveTrackedTimer TrackedPlayer)
    (CallFunction|ServerRequestClearDisplay (Utilities|Array|MakeArray TrackedPlayer))))

(fn TimerwasSuccessful
  (params (Timer%20Remaining :type Float%20%28double-precision%29))
  (return (or (> Timer%20Remaining 0.0) (Variables|UserOptions|GetSuccessOnTimerEnd))))

(fn UnbindDelegates
  (bind _fortminigameprogress (Variables|Components|GetFortMinigameProgress))
  (bind _exitedeventhandle (Variables|Default|GetExitedEventHandle))
  (Minigame|UnbindEventfromMinigameStateChangedDelegate _fortminigameprogress (EventDispatchers|CreateEvent :!event OnMinigameStateChanged))
  (Minigame|UnbindEventfromMinigameStartedDelegate _fortminigameprogress (EventDispatchers|CreateEvent :!event OnMinigameStarted))
  (Minigame|UnbindEventfromMinigameEndedDelegate _fortminigameprogress (EventDispatchers|CreateEvent :!event OnMinigameEnded))
  (Utilities|IsValid _exitedeventhandle
    (:Is%20Valid
      (Messaging|Unregister _exitedeventhandle))))

(fn UpdateStateForAll
  (params (Instigator :type Controller%20Object%20Reference) (NewState :type ETimerDeviceState%20Enum :default "Enabled") (Valid%20Transition%20States :type Set%20of%20ETimerDeviceState%20Enums :default "Enabled"))
  (switch Utilities|FlowControl|Switch|SwitchonE_Timer_AppliesTo (Variables|UserOptions|GetAppliesTo)
    (:NewEnumerator0
      (bind _bresult (Default|CanTriggerAllPersonalTimers Instigator))
      (if _bresult
        (bind _output_get (LocalVariables|SetTrackedPlayerList (CallFunction|GetTrackedPlayers)))
        (for _array_element _output_get
          (if (Utilities|Set|ContainsItem (Variables|GetValid%20Transition%20States) (CallFunction|GetTimerState _array_element))
            (CallFunction|TimerSetState _array_element (Variables|GetNewState))))))
    (:NewEnumerator1
      (if (Utilities|Set|ContainsItem Valid%20Transition%20States (CallFunction|GetTimerState 0))
        (CallFunction|TimerSetState 0 (Variables|GetNewState))))))

(fn UpdateTimerColor
  (params (Color :type Linear%20Color%20Structure))
  (Rendering|Components|TextRender|SetTextRenderColor (Variables|Components|GetTimerDisplayTextFront) (Math|Conversions|ToColor(LinearColor) Color))
  (Class|CreativeGenericTextWidgetComp|UpdateColor (Variables|Default|GetPrimaryTextWidget) (Variables|GetColor)))

(fn UpdateDisplayStateTextAndFX
  (params (CurrentState :type ETimerDeviceState%20Enum :default "Enabled") (TimeRemaining :type Float%20%28double-precision%29) (bFreshTick :type Boolean))
  (Networking|CanExecuteCosmeticEvents
    (:True
      (Utilities|FlowControl|Sequence
        (:then_0
          (switch Utilities|FlowControl|Switch|SwitchonETimerDeviceState (Variables|GetCurrentState)
            (:Enabled
              (LocalVariables|SetTextToDisplay (Utilities|String|ToString(Text) (Variables|UserOptions|GetTimerNotStartedText))))
            (:Activated
              (bind _output_get (LocalVariables|SetbInUrgency (and (Variables|UserOptions|GetEnableUrgencyMode) (<= (Variables|GetTimeRemaining) (Variables|UserOptions|GetUrgencyModeTime)))))
              (LocalVariables|SetTextToDisplay (Utilities|String|SelectString (Utilities|String|ToString(Text) (Variables|UserOptions|GetUrgencyText)) (Utilities|String|ToString(Text) (Variables|UserOptions|GetTimerRunningText)) (and _output_get (not (Utilities|String|IsEmpty (Utilities|String|ToString(Text) (Variables|UserOptions|GetUrgencyText))))))))
            (:Paused
              (LocalVariables|SetTextToDisplay (Utilities|String|ToString(Text) (Variables|UserOptions|GetPausedText))))
            (:Completed
              ;[1 1 1] Assumed display text
              (if (CallFunction|TimerWasSuccessful (Variables|GetTimeRemaining))
                (LocalVariables|SetTextToDisplay "SUCCESS")
                (else
                  (LocalVariables|SetTextToDisplay "FAILED"))))
            (:Disabled
              (LocalVariables|SetTextToDisplay "DISABLED"))))
        (:then_1
          (if (Utilities|String|NotEqualExactly(String) (Variables|FortCreativeTimerDevice|GetClientCurrentSecondaryText) (LocalVariables|GetTextToDisplay))
            (bind _output_get_2 (Variables|FortCreativeTimerDevice|SetClientCurrentSecondaryText (LocalVariables|GetTextToDisplay)))
            (Class|CreativeGenericTextWidgetComp|UpdateText (Variables|Default|GetSecondaryTextWidget) (Utilities|Text|ToText(String) _output_get_2))))
        (:then_2
          (if (LocalVariables|GetbInUrgency)
            (if (Variables|GetbFreshTick)
              (CallFunction|TriggerUrgentFX))
            (else
              (Audio|Components|Audio|Stop (Variables|Default|GetUrgentBuildAudio))
              (Variables|Cosmetic|SetbUrgencyHasStarted false)
              (Visual|UpdateTimerColor (Variables|UserOptions|GetTimerColor)))))))))

(fn UpdateDisplayTime
  (params (CurrentState :type ETimerDeviceState%20Enum :default "Enabled") (TimeRemaining :type Float%20%28double-precision%29))
  (bind _countdowndirection (Variables|UserOptions|Basic|GetCountDownDirection))
  (bind _returnvalue (Utilities|Enum|Equal(Enum) _countdowndirection "NewEnumerator0"))
  (bind _returnvalue_2 (Math|Float|Ceil TimeRemaining))
  (bind _duration (Utilities|GetCurrentDuration))
  (bind _returnvalue_3 (Math|Float|Truncate _duration))
  (bind _returnvalue_4 (- _returnvalue_3 _returnvalue_2))
  (bind _returnvalue_5 (Math|Integer|SelectInt _returnvalue_2 _returnvalue_4 _returnvalue))
  (bind _currentstate (Variables|GetCurrentState))
  (bind _duration_2 (Utilities|GetCurrentDuration))
  (bind _returnvalue_6 (Math|Float|Truncate _duration_2))
  (bind _returnvalue_7 (Utilities|Select :Index _currentstate :Enabled _returnvalue_6 :Activated _returnvalue_5 :Paused _returnvalue_5 :Completed _returnvalue_5 :Disabled 0))
  (bind _returnvalue_8 (Math|Integer|Clamp(Integer) _returnvalue_7 0 _returnvalue_6))
  (bind (_outminutes _outseconds) (Utilities|ConvertSecondstoMinutesSeconds (Variables|FortCreativeTimerDevice|GetClientCurrentTime)))
  (if (!= (Variables|FortCreativeTimerDevice|GetClientCurrentTime) _returnvalue_8)
    (Variables|FortCreativeTimerDevice|SetClientCurrentTime _returnvalue_8)
    (Rendering|Components|TextRender|SetText (Variables|Components|GetTimerDisplayTextFront) (Utilities|Select :Index (Variables|UserOptions|GetDisplayTimeIn) :NewEnumerator0 (Utilities|Text|FormatText "NSLOCTEXT(\"[91C0351C4A570D194F4D469F4C93D163]\", \"EAFF92EC4166F4591F0BEC87799C0F45\", \"{Min}:{Sec}\")" (Utilities|Text|ToText(Integer) _outminutes false true 2) (Utilities|Text|ToText(Integer) _outseconds false true 2)) :NewEnumerator7 (Utilities|Text|ToText(Integer) (Variables|FortCreativeTimerDevice|GetClientCurrentTime))))
    (return true)
    (else
      (return false))))

(fn UpdateInGameDisplay
  (params (CurrentState :type ETimerDeviceState%20Enum :default "Enabled") (TimeRemaining :type Float%20%28double-precision%29))
  (bind _bfreshtick (Visual|UpdateDisplayTime CurrentState TimeRemaining))
  (bind _output_get (Variables|FortCreativeTimerDevice|SetClientCurrentState (Variables|GetCurrentState)))
  (Visual|UpdateDisplayStateTextandFX _output_get (Variables|GetTimeRemaining) _bfreshtick)
  (if (Variables|UserOptions|GetShowOnHUD)
    (CallFunction|BroadcastUpdatetoHUD)))

(fn UpdateInGameVisibilities
  (Visibility|SetVisibilityComponents (Variables|Components|GetVisibleInGameComponent) (Utilities|Array|MakeArray (Variables|Components|GetTimerDisplayTextFront) (Variables|Components|GetTimerDisplayTextBack) (Variables|Default|GetUrgencyGlow)) true)
  (Visibility|SetVisiblityDuringPhase (Variables|Components|GetVisibleInGameComponent) (Utilities|Select :Index (Variables|UserOptions|GetVisibleDuringGame) :NewEnumerator0 "NewEnumerator1" :NewEnumerator1 "NewEnumerator3" :NewEnumerator2 "NewEnumerator3"))
  (Class|CreativeGenericTextWidgetComp|UpdateVisibilityandUpdateWidget (Variables|Default|GetPrimaryTextWidget) (Utilities|Select :Index (Variables|UserOptions|GetVisibleDuringGame) :NewEnumerator0 false :NewEnumerator1 false :NewEnumerator2 true)))

(fn ConstructionScript
  (|CallParentFunction))
