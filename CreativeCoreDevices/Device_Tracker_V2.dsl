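; Creative Tracker device that counts a chosen stat or event-driven progress toward a target, then can end the round or unassign itself.
; User options feed a CreativeQuest component, which holds progress per player, team or everyone and drives the HUD.
; Minigame callbacks and event handlers assign, adjust and reset progress.
; Optional persistence saves and loads progress, resolving shared-value conflicts by first player, average, median, max or min; legacy options are migrated.
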(event EventBeginPlay
  (Default|UpdateGameState)
  (Persistence|InitializePersistenceComponent (Variables|Default|GetFortCreativePersistenceComponent) (Variables|Default|GetFortMinigameProgress) true)
  (bind _returnvalue (Game|ConstructObjectfromClass (Variables|Persistence|GetPersistenceDataObjectClass)))
  (Variables|Persistence|SetPersistenceDataObject _returnvalue)
  (Networking|CanExecuteCosmeticEvents
    (:True
      (EventDispatchers|BindEventtoOnMessageLocalized (Variables|Default|GetFortLocalizableMessage) (delegate Custom|OnTextLocalized)))))

(event OnAnyPropertyChanged(ToyOptionsComponent)
  ;[0.1624 1 0.8499] Initialize Device
  (CallFunction|CopyProperties (Variables|Default|GetCreativeQuest)))

(event Custom|OnTextLocalized
  (CallFunction|UpdateQuestandObjectiveText))

(event OnPropertiesLoaded(ToyOptionsComponent)
  ;[0.1624 1 0.8499] Initialize Device
  (CallFunction|CopyProperties (Variables|Default|GetCreativeQuest)))

(event MinigameStateChangedDelegate(FortMinigameProgress)
  (params (Minigame :type Fort%20Minigame%20Object%20Reference) (NewMinigameState :type EFortMinigameState%20Enum :default "PreGame"))
  (Default|UpdateGameState)
  (switch Utilities|FlowControl|Switch|SwitchonEFortMinigameState NewMinigameState
    (:Setup
      (Networking|SwitchHasAuthority
        (:Authority
          (if (== (Variables|/Script/FortniteGame.FortMinigame|GetCurrentRound Minigame) 1)
            (Class|CreativeQuestComponent|ResetQuestProgress (Variables|Default|GetCreativeQuest) (Variables|UserOptions|Basic|GetStartingValue))
            (Default|SetMinigameTargetAmount (Variables|UserOptions|Basic|GetTargetValue))
            (Default|Unassign 0)))))
    (:InProgress
      (Networking|CanExecuteCosmeticEvents
        (:True
          (Utilities|FlowControl|Delay 0.0)
          (Class|CreativeQuestComponent|TriggerHUDUpdate (Variables|Default|GetCreativeQuest) true))))))

(event Custom|Increment%20Target%20Amount%20Event
  (params (Instigator :type Controller%20Object%20Reference))
  ;[0.2703 0 1] Gameplay Functions
  (Utilities|QuestIsActive Instigator)
  (Default|IncrementMinigameTargetAmount 1 Instigator))

(event Custom|Decrement%20Target%20Amount%20Event
  (params (Instigator :type Controller%20Object%20Reference))
  ;[0.2703 0 1] Gameplay Functions
  (Utilities|QuestIsActive Instigator)
  (Default|IncrementMinigameTargetAmount -1 Instigator))

(event OnConditionsMetDelegate(CreativeQuest)
  (params (ConditionInstigator :type Controller%20Object%20Reference) (IncrementAmount :type Integer))
  (Default|IncrementCurrentAmount ConditionInstigator IncrementAmount true))

(event Custom|Increment%20Progress%20Event
  (params (Instigator :type Controller%20Object%20Reference))
  ;[0.2703 0 1] Gameplay Functions
  (Default|IncrementCurrentAmount Instigator (Variables|UserOptions|GetAmountToChangeOnEvent)))

(event Custom|OnPawnJoinInProgress
  (params (Pawn :type Pawn%20Object%20Reference))
  (bind _returnvalue (Pawn|GetController Pawn))
  (Utilities|IsValid Pawn
    (:Is%20Valid
      (Utilities|IsValid _returnvalue
        (:Is%20Valid
          (if (Variables|UserOptions|Basic|GetAssignWhenJoiningInProgress)
            (CallFunction|AssignforPlayer _returnvalue)
            (Utilities|FlowControl|Sequence
              (:then_0
                _)
              (:then_1
                (CallFunction|ResetDistanceTimer)))
            (else
              _))
          (if (and (Variables|UserOptions|GetUsePersistence) (Variables|UserOptions|GetAuto-Load))
            (Persistence|LoadProgress _returnvalue true)))))))

(event Custom|Decrement%20Progress%20Event
  (params (Instigator :type Controller%20Object%20Reference))
  ;[0.2703 0 1] Gameplay Functions
  (Default|IncrementCurrentAmount Instigator (* (Variables|UserOptions|GetAmountToChangeOnEvent) -1)))

(event MinigameStartedDelegate(FortMinigameProgress)
  (if (Networking|HasAuthority)
    (Utilities|FlowControl|Sequence
      (:then_0
        (CallFunction|BindUnbindtoJoininProgressEvent true))
      (:then_1
        (if (Variables|UserOptions|Basic|GetAssignOnGameStart)
          (CallFunction|AssignforPlayer 0)))
      (:then_2
        (Class|CreativeQuestComponent|OnMinigameStarted (Variables|Default|GetCreativeQuest)))
      (:then_3
        (if (and (Variables|UserOptions|GetUsePersistence) (Variables|UserOptions|GetAuto-Load))
          (Persistence|LoadProgressforAll)))
      (:then_4
        (Persistence|SetAutoSaveEnabled (Variables|Default|GetFortCreativePersistenceComponent) (and (Variables|UserOptions|GetUsePersistence) (Variables|UserOptions|GetAuto-Save))))
      (:then_5
        (bind _returnvalue (GameplayEventRouter|GetEventRouter "/Script/FortniteGame.FortPlayspace" self)) ; ContextActor is how we bind to a specific playspace
        (bind (_asynctaskproxy _payload) (GameplayEventRouter|StartListeningToEvent _returnvalue "/Script/FortniteGame.FortTeamChangedContext")
          (:then
            (Variables|Default|SetTeamChangedObserver _asynctaskproxy))
          (:OnEventReceived
            (if (Variables|UserOptions|Basic|GetAssignOnGameStart)
              (bind _payload_player (Utilities|Struct|BreakFortTeamChangedContext _payload))
              (CallFunction|OnPawnJoininProgress (PlayerState|GetPawn _payload_player))))) ; Context here is irrelevant
      ))))

(event Custom|Assign%20Tracker%20Event
  (params (Instigator :type Controller%20Object%20Reference))
  ;[0.2703 0 1] Gameplay Functions
  (Utilities|IsValid Instigator
    (:Is%20Valid
      (Utilities|Set|Add (Variables|Default|GetAssignReceiverInstigators) Instigator)
      ;[1 1 1] Delay to next frame to allow this device to handle other events before this one as this can cause duplicate event triggers)
      (Utilities|FlowControl|Delay 0.0)
      (bind _result (Utilities|Set|ToArray (Variables|Default|GetAssignReceiverInstigators)))
      (for _array_element _result
        (CallFunction|AssignforPlayer _array_element))
      (Utilities|Set|Clear (Variables|Default|GetAssignReceiverInstigators)))))

(event Custom|Assign%20Tracker%20To%20All%20Event
  (params (Instigator :type Controller%20Object%20Reference))
  ;[0.2703 0 1] Gameplay Functions
  (CallFunction|AssignforPlayer 0))

(event Custom|Remove%20Tracker%20Event
  (params (Instigator :type Controller%20Object%20Reference))
  ;[0.2703 0 1] Gameplay Functions
  (Utilities|IsValid Instigator
    (:Is%20Valid
      (Default|Unassign Instigator))))

(event Custom|Remove%20Tracker%20From%20All%20Event
  (params (Instigator :type Controller%20Object%20Reference))
  ;[0.2703 0 1] Gameplay Functions
  (Default|Unassign 0))

(event MinigameEndedDelegate(FortMinigameProgress)
  (bind _fortlocalizablemessage (Variables|Default|GetFortLocalizableMessage))
  (if (Networking|HasAuthority)
    (Utilities|FlowControl|Sequence
      (:then_0
        (if (not (Class|CreativeQuestComponent|ShouldTrackBetweenRounds (Variables|Default|GetCreativeQuest)))
          (Class|CreativeQuestComponent|ResetQuestProgress (Variables|Default|GetCreativeQuest) (Variables|UserOptions|Basic|GetStartingValue))
          (Default|SetMinigameTargetAmount (Variables|UserOptions|Basic|GetTargetValue))
          (Default|Unassign 0)))
      (:then_1
        (CallFunction|BindUnbindtoJoininProgressEvent false))
      (:then_2
        (Persistence|SetAutoSaveEnabled (Variables|Default|GetFortCreativePersistenceComponent) false))
      (:then_3
        (Class|CreativeQuestComponent|OnMinigameEnded (Variables|Default|GetCreativeQuest)))
      (:then_4
        (Class|FortLocalizableMessageComponent|LocalizeandSetMessage _fortlocalizablemessage (Utilities|Struct|MakeLocalizableMessage) "DescriptionTextOverride" "DescriptionText")
        (Class|FortLocalizableMessageComponent|LocalizeandSetMessage _fortlocalizablemessage (Utilities|Struct|MakeLocalizableMessage) "TitleTextOverride" "TrackerTitle"))
      (:then_5
        (GameplayEventRouter|StopListeningtoEvent (Variables|Default|GetTeamChangedObserver))))))

(event Custom|Complete%20Tracker%20Event
  (params (Instigator :type Controller%20Object%20Reference))
  ;[0.2703 0 1] Gameplay Functions
  (Utilities|QuestIsActive Instigator)
  (Default|CheckTargetAmountReached Instigator true))

(event Custom|Reset%20Tracker%20Event
  (params (Instigator :type Controller%20Object%20Reference))
  ;[0.2703 0 1] Gameplay Functions
  (Default|SetCurrentAmount Instigator (Variables|UserOptions|Basic|GetStartingValue)))

(event Custom|Save%20Tracker%20Event
  (params (Instigator :type Controller%20Object%20Reference))
  ;[0.2703 0 1] Gameplay Functions
  (if (Variables|UserOptions|GetUsePersistence)
    (Persistence|SaveProgress Instigator)))

(event Custom|Save%20Tracker%20For%20All%20Event
  (params (Instigator :type Controller%20Object%20Reference))
  ;[0.2703 0 1] Gameplay Functions
  (if (Variables|UserOptions|GetUsePersistence)
    (Persistence|SaveProgress Instigator true)))

(event Custom|Load%20Tracker%20Event
  (params (Instigator :type Controller%20Object%20Reference))
  ;[0.2703 0 1] Gameplay Functions
  (if (Variables|UserOptions|GetUsePersistence)
    (Persistence|LoadProgress Instigator)))

(event EventOnWorldReady
  ;[0 1 0.08672] Visible In Game Component needs initialized even when working in basic mode as BP components can't initialize in editor,
  ; and BuildingProp based devices need to initialize visibleingame OnWorldReady
  (|CallParentFunction)
  (Class|CreativeVisibleinGameComponent|InitializeComponent (Variables|Default|GetVisibleInGameComponent)))

(event Custom|Load%20Tracker%20For%20All%20Event
  (params (Instigator :type Controller%20Object%20Reference))
  ;[0.2703 0 1] Gameplay Functions
  (if (Variables|UserOptions|GetUsePersistence)
    (Persistence|LoadProgressforAll)))

(event EditorPreview|EventSetupForEditorPreview
  ;[0 1 0.08672] Visible In Game Component needs initialized even when working in basic mode as BP components can't initialize in editor,
  ; and BuildingProp based devices need to initialize visibleingame OnWorldReady
  (|CallParentFunction)
  (Class|CreativeVisibleinGameComponent|InitializeComponent (Variables|Default|GetVisibleInGameComponent)))

(event Custom|Clear%20Persistence%20Event
  (params (Instigator :type Controller%20Object%20Reference))
  ;[0.2703 0 1] Gameplay Functions
  (if (Variables|UserOptions|GetUsePersistence)
    (Persistence|ClearPersistenceforPlayer Instigator)))

(event UserOptions|EventMigrateLegacyUserOptions
  (bind _reset_on_first_spawn (Class|GetClassDefaults "/CreativeCoreDevices/Device_Tracker_V2.Device_Tracker_V2_C"))
  (|CallParentFunction)
  (Utilities|FlowControl|Sequence
    (:then_0
      ; Migrate TeamSetting
      (if (not (== (Variables|Legacy|GetTeamSetting) (Class|GetClassDefaults "/CreativeCoreDevices/Device_Tracker_V2.Device_Tracker_V2_C")))
        (Creative|TeamOption|SetLegacyTeamIndex (Variables|UserOptions|GetTargetTeam) (Variables|Legacy|GetTeamSetting))
        (Variables|Legacy|SetTeamSetting (Class|GetClassDefaults "/CreativeCoreDevices/Device_Tracker_V2.Device_Tracker_V2_C"))
        (Variables|Default|SetMigrated%20Options true))
      ; Migrate ClassSetting
      (if (not (== (Variables|Legacy|GetClassSetting) (Class|GetClassDefaults "/CreativeCoreDevices/Device_Tracker_V2.Device_Tracker_V2_C")))
        (Creative|ClassOption|SetLegacyClassIndex (Variables|UserOptions|GetTargetClass) (Variables|Legacy|GetClassSetting))
        (Variables|Legacy|SetClassSetting (Class|GetClassDefaults "/CreativeCoreDevices/Device_Tracker_V2.Device_Tracker_V2_C"))
        (Variables|Default|SetMigrated%20Options true))
      ; Migrate AssignToTeam
      (if (not (== (Variables|Legacy|GetAssignedToTeam) (Class|GetClassDefaults "/CreativeCoreDevices/Device_Tracker_V2.Device_Tracker_V2_C")))
        (Creative|TeamOption|SetLegacyTeamIndex (Variables|UserOptions|GetValidTeam) (Variables|Legacy|GetAssignedToTeam))
        (Variables|Legacy|SetAssignedToTeam (Class|GetClassDefaults "/CreativeCoreDevices/Device_Tracker_V2.Device_Tracker_V2_C"))
        (Variables|Default|SetMigrated%20Options true))
      ; Migrate WinningTeamOnEndGame
      (if (not (== (Variables|Legacy|GetWinningTeamOnEndGame) (Class|GetClassDefaults "/CreativeCoreDevices/Device_Tracker_V2.Device_Tracker_V2_C")))
        (Creative|TeamOption|SetLegacyTeamIndex (Variables|UserOptions|GetWinningTeam) (Variables|Legacy|GetWinningTeamOnEndGame))
        (Variables|Legacy|SetWinningTeamOnEndGame (Class|GetClassDefaults "/CreativeCoreDevices/Device_Tracker_V2.Device_Tracker_V2_C"))
        (Variables|Default|SetMigrated%20Options true))
      ; Migrate WeaponCategory
      (if (not (== (Variables|Legacy|GetWeaponCategory) (Class|GetClassDefaults "/CreativeCoreDevices/Device_Tracker_V2.Device_Tracker_V2_C")))
        (switch int (Variables|Legacy|GetWeaponCategory)
          (:Default
            (Variables|UserOptions|SetTrackedWeaponCategory "Unspecified"))
          (:0
            (Variables|UserOptions|SetTrackedWeaponCategory "Unspecified"))
          (:1
            (Variables|UserOptions|SetTrackedWeaponCategory "FallDamage"))
          (:2
            (Variables|UserOptions|SetTrackedWeaponCategory "Pistol"))
          (:3
            (Variables|UserOptions|SetTrackedWeaponCategory "SMG"))
          (:4
            (Variables|UserOptions|SetTrackedWeaponCategory "Shotgun"))
          (:5
            (Variables|UserOptions|SetTrackedWeaponCategory "Rifle"))
          (:6
            (Variables|UserOptions|SetTrackedWeaponCategory "Sniper"))
          (:7
            (Variables|UserOptions|SetTrackedWeaponCategory "GrenadeLauncher"))
          (:8
            (Variables|UserOptions|SetTrackedWeaponCategory "RocketLauncher"))
          (:9
            (Variables|UserOptions|SetTrackedWeaponCategory "Bow"))
          (:10
            (Variables|UserOptions|SetTrackedWeaponCategory "Melee")))
        (Variables|Legacy|SetWeaponCategory (Class|GetClassDefaults "/CreativeCoreDevices/Device_Tracker_V2.Device_Tracker_V2_C"))
        (Variables|Default|SetMigrated%20Options true))
      ; Migrate ResetBetweenRounds
      (if (not (== (Variables|Legacy|GetResetBetweenRounds) _reset_on_first_spawn))
        (Variables|UserOptions|SetReset%20on%20First%20Spawn (Variables|Legacy|GetResetBetweenRounds))
        (Variables|Legacy|SetResetBetweenRounds _reset_on_first_spawn) ; Set the legacy value to the new default so we don't migrate again
        (Variables|Default|SetMigrated%20Options true)))
    (:then_1
      (if (Variables|Default|GetMigrated%20Options)
        (OptionData|RefreshPropertyOverrides (Variables|Default|GetToyOptionsComponent))))))

(event Custom|Clear%20Persistence%20For%20All%20Event
  ;[0.2703 0 1] Gameplay Functions
  (if (Variables|UserOptions|GetUsePersistence)
    (bind _playerstate (Utilities|ForEachPersistentPlayer)
      (:LoopBody
        (Persistence|ClearPersistenceforPlayer (CreativeAI|Controller|GetParticipantControllerfromActor _playerstate false))))))

(event EventPostFinishSpawningInEditor
  (|CallParentFunction)
  (if (not (Variables|Default|GetMigratedPersistence))
    (Variables|UserOptions|SetUsePersistence (Math|Boolean|ORBoolean (Class|FortGameplayMessageComponentBase|IsChannelIdSet (Variables|Default|GetSaveProgressReceiver)) (Class|FortGameplayMessageComponentBase|IsChannelIdSet (Variables|Default|GetLoadProgressForPlayerReceiver)) (Class|FortGameplayMessageComponentBase|IsChannelIdSet (Variables|Default|GetLoadProgressForAllReceiver)) (Class|FortGameplayMessageComponentBase|IsChannelIdSet (Variables|Default|GetClearPersistenceReceiver)) (Variables|UserOptions|GetAuto-Load) (Variables|UserOptions|GetAuto-Save) (Variables|UserOptions|GetUsePersistence)))
    (Variables|Default|SetMigratedPersistence true)
    (OptionData|RefreshPropertyOverrides (Variables|Default|GetToyOptionsComponent))
    (Class|FortCreativePersistenceComponentBase|UpdateIsPersistencePosible (Variables|Default|GetFortCreativePersistenceComponent))))

(event EventPostFinishSpawning
  (|CallParentFunction)
  (if (not (Variables|Default|GetMigratedPersistence))
    (Variables|UserOptions|SetUsePersistence (Math|Boolean|ORBoolean (Class|FortGameplayMessageComponentBase|IsChannelIdSet (Variables|Default|GetSaveProgressReceiver)) (Class|FortGameplayMessageComponentBase|IsChannelIdSet (Variables|Default|GetLoadProgressForPlayerReceiver)) (Class|FortGameplayMessageComponentBase|IsChannelIdSet (Variables|Default|GetLoadProgressForAllReceiver)) (Class|FortGameplayMessageComponentBase|IsChannelIdSet (Variables|Default|GetClearPersistenceReceiver)) (Variables|UserOptions|GetAuto-Load) (Variables|UserOptions|GetAuto-Save) (Variables|UserOptions|GetUsePersistence)))
    (Variables|Default|SetMigratedPersistence true)
    (OptionData|RefreshPropertyOverrides (Variables|Default|GetToyOptionsComponent))
    (Class|FortCreativePersistenceComponentBase|UpdateIsPersistencePosible (Variables|Default|GetFortCreativePersistenceComponent))))

(fn ApplyLoadedData
  (params (bActive :type Boolean) (Progress :type Integer) (Target :type Integer) (Player :type Controller%20Object%20Reference))
  (Class|CreativeQuestComponent|SetActiveFor (Variables|Default|GetCreativeQuest) (Variables|GetPlayer) (Variables|GetbActive))
  (Default|SetCurrentAmount (Variables|GetPlayer) (Variables|GetProgress))
  (Default|SetMinigameTargetAmount (Variables|GetTarget) (Variables|GetPlayer))
  (GameplayEvent|BroadcastEvent (Variables|UserOptions|Events|GetOn%20Loaded) 0 (Utilities|Struct|MakeGameplayEventonLoaded (Variables|GetPlayer))))

(fn AssignForPlayer
  (params (Instigator :type Controller%20Object%20Reference))
  (Class|CreativeQuestComponent|SetActiveFor (Variables|Default|GetCreativeQuest) Instigator true)
  (Default|CheckTargetAmountReached Instigator))

(fn BindUnbindToJoinInProgressEvent
  (params (Bind :type Boolean))
  (bind _outputdelegate (EventDispatchers|CreateEvent :!event OnPawnJoinInProgress))
  (bind _fortminigameprogress (Variables|Default|GetFortMinigameProgress))
  (bind _currentminigame (Variables|/Script/FortniteGame.FortMinigameLogicComponent|GetCurrentMinigame _fortminigameprogress))
  (Utilities|IsValid _currentminigame
    (:Is%20Valid
      (if (Variables|GetBind)
        (Minigame|BindEventtoOnPlayerJoininProgress _currentminigame _outputdelegate)
        (else
          (Minigame|UnbindEventfromOnPlayerJoininProgress _currentminigame _outputdelegate))))))

(fn CanPlayerAutoSave
  (params (FortPlayerState :type Fort%20Player%20State%20Object%20Reference))
  (return true))

(fn CheckTargetAmountReached
  (params (Instigator :type Controller%20Object%20Reference) (bForceComplete :type Boolean))
  (bind _instigator (Variables|Actor|GetInstigator))
  (bind _creativequest (Variables|Default|GetCreativeQuest))
  (bind _returnvalue (Class|CreativeQuestComponent|GetActiveSharedControllers _creativequest _instigator))
  (bind _instigator_2 (Variables|Actor|GetInstigator))
  (bind (_winning_team _team_index) (CallFunction|WinningTeamType _instigator_2))
  (bind _fortminigameprogress (Variables|Default|GetFortMinigameProgress))
  (bind _currentminigame (Variables|/Script/FortniteGame.FortMinigameLogicComponent|GetCurrentMinigame _fortminigameprogress))
  (if (or (Variables|GetbForceComplete) (and (>= (Default|GetCurrentAmount Instigator) (Variables|Default|GetMinigameTargetAmount)) (> (Variables|Default|GetMinigameTargetAmount) 0)))
    (switch Utilities|FlowControl|Switch|SwitchonETargetReachedBehavior (Variables|UserOptions|GetWhenTargetIsReached)
      (:NewEnumerator0
        (Utilities|IsValid _currentminigame
          (:Is%20Valid
            (if (Minigame|IsinProgress _currentminigame)
              (Utilities|FlowControl|Sequence
                (:then_0
                  (if (Utilities|Enum|NotEqual(Enum) _winning_team "NewEnumerator0")
                    (Minigame|SetWinningTeamIndex _currentminigame (Minigame|GetWinningTeamIndexfromTeamIndex _currentminigame _team_index))))
                (:then_1
                  (Minigame|EndGame _currentminigame "EndRound")))))))
      (:NewEnumerator1
        _)
      (:NewEnumerator2
        (for _array_element _returnvalue
          (Default|Unassign _array_element))))
    (for _array_element_2 _returnvalue
      (GameplayEvent|BroadcastEvent (Variables|UserOptions|Events|GetWhen%20Complete) 0 (Utilities|Struct|MakeGameplayEventQuestComplete _array_element_2)))))

(fn ClearPersistenceForPlayer
  (params (Instigator :type Controller%20Object%20Reference))
  (bind _playerstate (Variables|/Script/Engine.Controller|GetPlayerState Instigator))
  (Utilities|IsValid Instigator
    (:Is%20Valid
      (Utilities|IsValid _playerstate
        (:Is%20Valid
          (Persistence|ClearData (Variables|Default|GetFortCreativePersistenceComponent) (PlayerState|GetUniqueNetId _playerstate) (Variables|Persistence|GetSavePath))
          (GameplayEvent|BroadcastEvent (Variables|UserOptions|Events|GetOn%20Cleared) 0 (Utilities|Struct|MakeGameplayEventonCleared Instigator)))))))

(fn CopyProperties
  (params (Component :type Creative%20Quest%20Component%20Object%20Reference))
  (Utilities|IsValid Component
    (:Is%20Valid
      (Utilities|FlowControl|Sequence
        (:then_0
          (Variables|/Script/FortniteGame.CreativeQuestComponent|SetTargetTeam (Creative|TeamOption|GetLegacyTeamIndex (Variables|UserOptions|GetTargetTeam)) (Variables|GetComponent))
          (Variables|/Script/FortniteGame.CreativeQuestComponent|SetTargetClass (Creative|ClassOption|GetLegacyClassIndex (Variables|UserOptions|GetTargetClass)) (Variables|GetComponent))
          (Variables|/Script/FortniteGame.CreativeQuestComponent|SetTargetDeathCause (Variables|UserOptions|GetTrackedWeaponCategory) (Variables|GetComponent))
          (Variables|/Script/FortniteGame.CreativeQuestComponent|SetStatToTrack (Variables|UserOptions|Basic|GetStatToTrack) (Variables|GetComponent))
          (Variables|/Script/FortniteGame.CreativeQuestComponent|SetbResetBetweenRounds (Variables|UserOptions|GetReset%20on%20First%20Spawn) (Variables|GetComponent)))
        (:then_1
          (Variables|Default|SetMinigameTargetAmount (Variables|UserOptions|Basic|GetTargetValue))
          (Variables|/Script/FortniteGame.CreativeQuestComponent|SetbSelfEliminationsCount (Variables|UserOptions|GetSelfEliminationsCount) (Variables|GetComponent))
          (Variables|/Script/FortniteGame.CreativeQuestComponent|SetSharingMode (Variables|UserOptions|GetSharing) (Variables|GetComponent))
          (Variables|/Script/FortniteGame.CreativeQuestComponent|SetAssignedToTeam (Creative|TeamOption|GetLegacyTeamIndex (Variables|UserOptions|GetValidTeam)) (Variables|GetComponent)))
        (:then_2
          (Variables|/Script/FortniteGame.CreativeQuestComponent|SetQuestName (Utilities|String|ToString(Text) (Variables|UserOptions|Basic|GetTrackerTitle)) (Variables|GetComponent))
          (Variables|/Script/FortniteGame.CreativeQuestComponent|SetObjectiveText (Utilities|String|ToString(Text) (Variables|UserOptions|Basic|GetDescriptionText)) (Variables|GetComponent))
          (Variables|/Script/FortniteGame.CreativeQuestComponent|SetMinigameTargetAmount (Variables|Default|GetMinigameTargetAmount) (Variables|GetComponent))
          (Variables|/Script/FortniteGame.CreativeQuestComponent|SetShowProgressMode (Variables|UserOptions|GetShowProgress) (Variables|GetComponent))
          (Variables|/Script/FortniteGame.CreativeQuestComponent|SetShowQuestOnHUD (Variables|UserOptions|GetShowOnHUD) (Variables|GetComponent)))
        (:then_3
          (Variables|/Script/FortniteGame.CreativeQuestComponent|SetbShowCompleteEffects (Variables|UserOptions|GetTrackerCompletionCeremony) (Variables|GetComponent))
          (Variables|/Script/FortniteGame.CreativeQuestComponent|SetbHeadshotsOnly (Variables|UserOptions|GetHeadshotsOnly) (Variables|GetComponent))
          (Variables|/Script/FortniteGame.CreativeQuestComponent|SetbPlayerEliminationsCount (Variables|UserOptions|GetPlayerEliminationsCount) (Variables|GetComponent))
          (Variables|/Script/FortniteGame.CreativeQuestComponent|SetbNonPlayerEliminationsCount (Variables|UserOptions|GetNonPlayerEliminationsCount) (Variables|GetComponent)))
        (:then_4
          (Variables|/Script/FortniteGame.CreativeQuestComponent|SetQuestIconImage (Creative|UI|IconPicker|GetIconLibraryTexture (Variables|UserOptions|GetQuestIcon)) (Variables|GetComponent))
          (Variables|/Script/FortniteGame.CreativeQuestComponent|SetQuestColor (Variables|UserOptions|GetColor) (Variables|GetComponent))
          (Variables|/Script/FortniteGame.CreativeQuestComponent|SetQuestWidget (Variables|UserOptions|Basic|GetHUD%20Widget) (Variables|GetComponent))
          (Networking|CanExecuteCosmeticEvents
            (:True
              (Class|CreativeQuestComponent|OnRepQuestUpdate (Variables|GetComponent)))))
        (:then_5
          (Variables|/Script/FortniteGame.CreativeQuestComponent|SetbNonPlayerDamageCount (Variables|UserOptions|GetNonPlayerDamageCount) (Variables|GetComponent))
          (Variables|/Script/FortniteGame.CreativeQuestComponent|SetbPlayerDamageCount (Variables|UserOptions|GetPlayerDamageCount) (Variables|GetComponent))
          (Variables|/Script/FortniteGame.CreativeQuestComponent|SetbSelfDamageCount (Variables|UserOptions|GetSelfDamageCount) (Variables|GetComponent)))))))

(fn GetFirstPlayerProgress
  (params (Instigator :type Controller%20Object%20Reference))
  (Utilities|FlowControl|Sequence
    (:then_0
      (LocalVariables|SetSharedControllers (Class|CreativeQuestComponent|GetActiveSharedControllers (Variables|Default|GetCreativeQuest) (Variables|Actor|GetInstigator))))
    (:then_1
      (for _array_element (Minigame|GetPlayersOrdered (Variables|/Script/FortniteGame.FortMinigameLogicComponent|GetCurrentMinigame (Variables|Default|GetFortMinigameProgress)))
        (if (Utilities|Array|ContainsItem (LocalVariables|GetSharedControllers) (CreativeAI|Controller|GetParticipantControllerfromActor _array_element))
          (bind (_bloadsuccess _loadeddata) (Persistence|LoadToPlayer (CreativeAI|Controller|GetParticipantControllerfromActor _array_element)))
          (if _bloadsuccess
            (return (Variables|/Game/Creative/Devices/Quests/PersistenceData_QuestDevice.PersistenceData_QuestDevice_C|GetProgress _loadeddata) (Variables|/Game/Creative/Devices/Quests/PersistenceData_QuestDevice.PersistenceData_QuestDevice_C|GetTarget _loadeddata)))))
      (return 0 (Variables|UserOptions|Basic|GetTargetValue)))))

(fn GetActiveAgents
  (bind _currentminigame (Variables|/Script/FortniteGame.FortMinigameLogicComponent|GetCurrentMinigame (Variables|Default|GetFortMinigameProgress))
    (:then
      (bind _outplayers (Minigame|GetParticipatingPlayers_NotPure _currentminigame))
      (for _array_element _outplayers
        (Utilities|IsValid _array_element
          (:Is%20Valid
            (Utilities|IsValid (Fortnite|GetControllerfromActor _array_element)
              (:Is%20Valid
                (if (Class|CreativeQuestComponent|IsActiveFor (Variables|Default|GetCreativeQuest) (Fortnite|GetControllerfromActor _array_element))
                  (Utilities|Array|Add (LocalVariables|GetActiveControllers) (Fortnite|GetControllerfromActor _array_element))))))))
      (return (LocalVariables|GetActiveControllers)))
    (:else)))

(fn GetCurrentAmount
  (params (Controller :type Controller%20Object%20Reference))
  (return (Class|CreativeQuestComponent|GetCurrentAmount (Variables|Default|GetCreativeQuest) Controller)))

(fn GetDescriptionText
  (return (Utilities|String|ToString(Text) (select (Utilities|Text|TextIsEmpty (Variables|Verse|GetDescriptionTextOverride)) (Variables|UserOptions|Basic|GetDescriptionText) (Variables|Verse|GetDescriptionTextOverride)))))

(fn GetObjectToAutoSaveForPlayer
  (params (FortPlayerState :type Fort%20Player%20State%20Object%20Reference))
  (Persistence|PopulatePersistenceDataObject (CreativeAI|Controller|GetParticipantControllerfromActor FortPlayerState false))
  (return (Variables|Persistence|GetPersistenceDataObject)))

(fn GetPersistenceRequiredState
  (return (Variables|UserOptions|GetUsePersistence)))

(fn GetTargetValue
  (return (Variables|Default|GetMinigameTargetAmount)))

(fn GetTitleText
  (return (Utilities|String|ToString(Text) (select (Utilities|Text|TextIsEmpty (Variables|Verse|GetTitleTextOverride)) (Variables|UserOptions|Basic|GetTrackerTitle) (Variables|Verse|GetTitleTextOverride)))))

(fn GetValue
  (params (Controller :type Controller%20Object%20Reference))
  (return (Default|GetCurrentAmount Controller)))

(fn GetValueForAll
  (if (Utilities|Enum|Equal(Enum) (Variables|UserOptions|GetSharing) "Team")
    ;[0 0.6864 0.7396 0.292] If we're Sharing::Team, and we ask for All Score, iterate through each team and sum their total score.
    (bind _currentminigame (Variables|/Script/FortniteGame.FortMinigameLogicComponent|GetCurrentMinigame (Variables|Default|GetFortMinigameProgress))
      (:then
        (Utilities|IsValid (Minigame|GetMinigameTeamComponent _currentminigame)
          (:Is%20Valid
            (for _array_element (Variables|/Script/FortniteGame.FortMinigameTeamComponent|GetTeamArray (Minigame|GetMinigameTeamComponent _currentminigame))
              (if (.bHasBucketAvailable (Utilities|Struct|BreakMinigameTeam _array_element))
                (bind _result (ArrayMacros|GetValidIndex (Class|FortMinigameTeamComponent|GetTeamMembers (Minigame|GetMinigameTeamComponent _currentminigame) (Utilities|Struct|BreakMinigameTeam _array_element)))
                  (:Valid%20Index
                    (Utilities|IsValid _result
                      (:Is%20Valid
                        (LocalVariables|SetAllValue (+ (Default|GetCurrentAmount (Fortnite|GetControllerfromActor _result)) (LocalVariables|GetAllValue)))))))))
            (return (LocalVariables|GetAllValue)))
          (:Is%20Not%20Valid
            _)))
      (:else
        _))
    (return 0)
    (else
      (bind _currentminigame_2 (Variables|/Script/FortniteGame.FortMinigameLogicComponent|GetCurrentMinigame (Variables|Default|GetFortMinigameProgress))
        (:then
          (bind _outplayers (Minigame|GetParticipatingPlayers_NotPure _currentminigame_2))
          (bind _array_element_2 (Utilities|Array|ForEachLoopwithBreak _outplayers)
            (:LoopBody
              (Utilities|IsValid _array_element_2
                (:Is%20Valid
                  (LocalVariables|SetAllValue (+ (Default|GetCurrentAmount (Fortnite|GetControllerfromActor _array_element_2)) (LocalVariables|GetAllValue)))
                  (if (Utilities|Enum|Equal(Enum) (Variables|UserOptions|GetSharing) "All")
                    _))))
            (:Completed
              _))
          (return (LocalVariables|GetAllValue)))
        (:else
          (return 0))))))

(fn GetValueForTeam
  (params (TeamIndex :type Integer))
  (if (Utilities|Enum|Equal(Enum) (Variables|UserOptions|GetSharing) "All")
    (bind _value (Verse|GetValueforAll))
    (return _value)
    (else
      (bind _currentminigame (Variables|/Script/FortniteGame.FortMinigameLogicComponent|GetCurrentMinigame (Variables|Default|GetFortMinigameProgress))
        (:then
          (Utilities|IsValid (Minigame|GetMinigameTeamComponent _currentminigame)
            (:Is%20Valid
              (bind _output_get (Variables|Team|SetTeam (Class|FortMinigameTeamComponent|GetTeamMembers (Minigame|GetMinigameTeamComponent _currentminigame) (Teams|IndextoTeam (- TeamIndex 1)))))
              (bind _array_element (Utilities|Array|ForEachLoopwithBreak _output_get)
                (:LoopBody
                  (Utilities|IsValid _array_element
                    (:Is%20Valid
                      (bind _value_2 (Verse|GetValue (CreativeAI|Controller|GetParticipantControllerfromActor _array_element false)))
                      (LocalVariables|SetTeamValue (+ _value_2 (LocalVariables|GetTeamValue)))
                      (if (Utilities|Enum|Equal(Enum) (Variables|UserOptions|GetSharing) "Team")
                        _))))
                (:Completed
                  _))
              (return (LocalVariables|GetTeamValue)))
            (:Is%20Not%20Valid
              (return 0))))
        (:else
          (return 0))))))

(fn HasReachedTarget
  (params (Controller :type Controller%20Object%20Reference))
  (Utilities|IsValid Controller
    (:Is%20Valid
      (return (and (>= (Default|GetCurrentAmount Controller) (Variables|Default|GetMinigameTargetAmount)) (> (Variables|Default|GetMinigameTargetAmount) 0))))
    (:Is%20Not%20Valid
      (return false))))

(fn IncrementCurrentAmount
  (params (Controller :type Controller%20Object%20Reference) (IncrementBy :type Integer) (BypassActiveCheck :type Boolean))
  ;[0.4948 0.3155 0 0.5] FORT-873165 In Newton islands it is possible to receive a Prop and not a controller as Instigator, meanwhile we don't have a propper implementation for props we just ignore them
  (if (Fortnite|IsPhysicallySimulatedIsland)
    (Utilities|Casting|CastToController Controller))
  (if (not BypassActiveCheck)
    (Utilities|QuestIsActive Controller))
  (Default|SetCurrentAmount (Variables|GetController) (+ (Default|GetCurrentAmount Controller) IncrementBy)))

(fn IncrementMinigameTargetAmount
  (params (IncrementBy :type Integer :default 1) (Instigator :type Controller%20Object%20Reference))
  (Default|SetMinigameTargetAmount (+ (Variables|Default|GetMinigameTargetAmount) IncrementBy) Instigator))

(fn IsActive
  (params (Controller :type Controller%20Object%20Reference))
  (Utilities|IsValid Controller
    (:Is%20Valid
      (return (Class|CreativeQuestComponent|IsActiveFor (Variables|Default|GetCreativeQuest) Controller)))
    (:Is%20Not%20Valid
      (return false))))

(fn LoadForSharedControllersAndDisambiguate
  (params (Instigator :type Controller%20Object%20Reference))
  (switch Utilities|FlowControl|Switch|SwitchonECreativeQuest_ResolveConflicts (Variables|UserOptions|GetResolvesConflicts)
    (:NewEnumerator0
      (bind (_progress _target) (Persistence|GetFirstPlayerProgress (Variables|Actor|GetInstigator)))
      (return _progress _target))
    (:NewEnumerator1
      (bind (_outprogresslist _outtargetlist) (Persistence|LoadProgressfromSharedControllers (Variables|Actor|GetInstigator)))
      (return (Math|Float|Round (Math|Integer|AverageOfIntArray _outprogresslist)) (Math|Float|Round (Math|Integer|AverageOfIntArray _outtargetlist))))
    (:NewEnumerator2
      (bind (_outprogresslist_2 _outtargetlist_2) (Persistence|LoadProgressfromSharedControllers (Variables|Actor|GetInstigator)))
      (return (Math|Float|Round (Math|Integer|MedianOfIntArray _outprogresslist_2)) (Math|Float|Round (Math|Integer|MedianOfIntArray _outtargetlist_2))))
    (:NewEnumerator3
      (bind (_outprogresslist_3 _outtargetlist_3) (Persistence|LoadProgressfromSharedControllers (Variables|Actor|GetInstigator)))
      (return (.MaxValue (Math|Integer|MaxOfIntArray _outprogresslist_3)) (.MaxValue (Math|Integer|MaxOfIntArray _outtargetlist_3))))
    (:NewEnumerator4
      (bind (_outprogresslist_4 _outtargetlist_4) (Persistence|LoadProgressfromSharedControllers (Variables|Actor|GetInstigator)))
      (return (.MinValue (Math|Integer|MinOfIntArray _outprogresslist_4)) (.MinValue (Math|Integer|MinOfIntArray _outtargetlist_4))))))

(fn LoadProgress
  (params (Instigator :type Controller%20Object%20Reference) (bIsJIP :type Boolean))
  (bind _instigator (Variables|Actor|GetInstigator))
  (bind _creativequest (Variables|Default|GetCreativeQuest))
  (bind _returnvalue (Class|CreativeQuestComponent|GetActiveSharedControllers _creativequest _instigator))
  (Utilities|IsValid Instigator
    (:Is%20Valid
      (switch Utilities|FlowControl|Switch|SwitchonECreativeQuestSharing (Variables|UserOptions|GetSharing)
        (:Individual
          (bind (_bloadsuccess _loadeddata) (Persistence|LoadToPlayer (Variables|Actor|GetInstigator)))
          (if _bloadsuccess
            (Persistence|ApplyLoadedData (Variables|/Game/Creative/Devices/Quests/PersistenceData_QuestDevice.PersistenceData_QuestDevice_C|GetbActive _loadeddata) (Variables|/Game/Creative/Devices/Quests/PersistenceData_QuestDevice.PersistenceData_QuestDevice_C|GetProgress _loadeddata) (Variables|/Game/Creative/Devices/Quests/PersistenceData_QuestDevice.PersistenceData_QuestDevice_C|GetTarget _loadeddata) (Variables|Actor|GetInstigator))))
        (:Team
          _)
        (:All
          _))
      (if (not (Variables|GetbIsJIP))
        _
        (elif (Variables|UserOptions|GetResolveConflictAfterTrackerActive)
          _
          (elif (Utilities|Array|IsNotEmpty _returnvalue)
            (Class|CreativeQuestComponent|SetCurrentAmount (Variables|Default|GetCreativeQuest) (Variables|Actor|GetInstigator) (Class|CreativeQuestComponent|GetCurrentAmount (Variables|Default|GetCreativeQuest) (Utilities|Array|Get(acopy) _returnvalue)))
            (else
              (bind (_bloadsuccess_2 _loadeddata_2) (Persistence|LoadToPlayer (Variables|Actor|GetInstigator)))
              (if _bloadsuccess_2
                (Persistence|ApplyLoadedData (Variables|/Game/Creative/Devices/Quests/PersistenceData_QuestDevice.PersistenceData_QuestDevice_C|GetbActive _loadeddata_2) (Variables|/Game/Creative/Devices/Quests/PersistenceData_QuestDevice.PersistenceData_QuestDevice_C|GetProgress _loadeddata_2) (Variables|/Game/Creative/Devices/Quests/PersistenceData_QuestDevice.PersistenceData_QuestDevice_C|GetTarget _loadeddata_2) (Variables|Actor|GetInstigator)))))))
      (bind (_progress _target) (Persistence|LoadforSharedControllersandDisambiguate (Variables|Actor|GetInstigator)))
      (for _array_element (Class|CreativeQuestComponent|GetActiveSharedControllers (Variables|Default|GetCreativeQuest) (Variables|Actor|GetInstigator))
        (bind (_bloadsuccess_3 _loadeddata_3) (Persistence|LoadToPlayer _array_element))
        (Persistence|ApplyLoadedData (select _bloadsuccess_3 (Variables|/Game/Creative/Devices/Quests/PersistenceData_QuestDevice.PersistenceData_QuestDevice_C|GetbActive _loadeddata_3) (Class|CreativeQuestComponent|IsActiveFor (Variables|Default|GetCreativeQuest) _array_element)) _progress _target _array_element)))))

(fn LoadProgressForAll
  (bind _fortminigameprogress (Variables|Default|GetFortMinigameProgress))
  (bind _currentminigame (Variables|/Script/FortniteGame.FortMinigameLogicComponent|GetCurrentMinigame _fortminigameprogress))
  (Utilities|IsValid _currentminigame
    (:Is%20Valid
      (bind _persistencemanagercomponent (Variables|/Script/FortniteGame.FortMinigame|GetPersistenceManagerComponent _currentminigame)
        (:then
          (LocalVariables|SetPersistencePlayers (Persistence|GetPersistencePlayers _persistencemanagercomponent :!class /Script/FortniteGame.FNEPersistenceManagerComponent))
          (switch Utilities|FlowControl|Switch|SwitchonECreativeQuestSharing (Variables|UserOptions|GetSharing)
            (:Individual
              (for _array_element (LocalVariables|GetPersistencePlayers)
                (bind _output_get (LocalVariables|SetCurrentPlayer (CreativeAI|Controller|GetParticipantControllerfromActor _array_element false)))
                (bind (_bloadsuccess _loadeddata) (Persistence|LoadToPlayer _output_get))
                (if _bloadsuccess
                  (Persistence|ApplyLoadedData (Variables|/Game/Creative/Devices/Quests/PersistenceData_QuestDevice.PersistenceData_QuestDevice_C|GetbActive _loadeddata) (Variables|/Game/Creative/Devices/Quests/PersistenceData_QuestDevice.PersistenceData_QuestDevice_C|GetProgress _loadeddata) (Variables|/Game/Creative/Devices/Quests/PersistenceData_QuestDevice.PersistenceData_QuestDevice_C|GetTarget _loadeddata) (LocalVariables|GetCurrentPlayer)))))
            (:Team
              (for _array_element_2 (LocalVariables|GetPersistencePlayers)
                (LocalVariables|SetCurrentPlayer (CreativeAI|Controller|GetParticipantControllerfromActor _array_element_2 false))
                (if (not (Utilities|Array|ContainsItem (LocalVariables|GetAlreadyLoadedPlayers) (LocalVariables|GetCurrentPlayer)))
                  (bind (_progress _target) (Persistence|LoadforSharedControllersandDisambiguate (LocalVariables|GetCurrentPlayer)))
                  (for _array_element_3 (Class|CreativeQuestComponent|GetActiveSharedControllers (Variables|Default|GetCreativeQuest) (LocalVariables|GetCurrentPlayer))
                    (bind (_bloadsuccess_2 _loadeddata_2) (Persistence|LoadToPlayer _array_element_3))
                    (Persistence|ApplyLoadedData (select _bloadsuccess_2 (Variables|/Game/Creative/Devices/Quests/PersistenceData_QuestDevice.PersistenceData_QuestDevice_C|GetbActive _loadeddata_2) (Class|CreativeQuestComponent|IsActiveFor (Variables|Default|GetCreativeQuest) _array_element_3)) _progress _target _array_element_3)
                    (Utilities|Array|Add (LocalVariables|GetAlreadyLoadedPlayers) _array_element_3)))))
            (:All
              (bind _array_element_4 (Utilities|Array|ForEachLoopwithBreak (LocalVariables|GetPersistencePlayers))
                (:LoopBody
                  (if (CallFunction|PlayerTeamApplies _array_element_4)
                    (LocalVariables|SetCurrentPlayer (CreativeAI|Controller|GetParticipantControllerfromActor _array_element_4 false))))
                (:Completed
                  _))
              (Utilities|IsValid (LocalVariables|GetCurrentPlayer)
                (:Is%20Valid
                  _)))))
        (:else)))))

(fn LoadProgressFromSharedControllers
  (params (Instigator :type Controller%20Object%20Reference))
  (for _array_element (Class|CreativeQuestComponent|GetActiveSharedControllers (Variables|Default|GetCreativeQuest) Instigator)
    (bind (_bloadsuccess _loadeddata) (Persistence|LoadToPlayer _array_element))
    (Utilities|Array|Add (LocalVariables|GetProgressList) (select _bloadsuccess (Variables|/Game/Creative/Devices/Quests/PersistenceData_QuestDevice.PersistenceData_QuestDevice_C|GetProgress _loadeddata) 0))
    (Utilities|Array|Add (LocalVariables|GetTargetList) (select _bloadsuccess (Variables|/Game/Creative/Devices/Quests/PersistenceData_QuestDevice.PersistenceData_QuestDevice_C|GetTarget _loadeddata) (Variables|UserOptions|Basic|GetTargetValue))))
  (return (LocalVariables|GetProgressList) (LocalVariables|GetTargetList)))

(fn LoadtoPlayer
  (params (Controller :type Controller%20Object%20Reference))
  (bind _persistencedataobject (Variables|Persistence|GetPersistenceDataObject))
  (Utilities|IsValid Controller
    (:Is%20Valid
      (Persistence|ResetPersistenceObject)
      (Persistence|LoadData (Variables|Default|GetFortCreativePersistenceComponent) (PlayerState|GetUniqueNetId (Variables|/Script/Engine.Controller|GetPlayerState Controller)) (Variables|Persistence|GetSavePath) _persistencedataobject)
      (if (Variables|/Game/Creative/Devices/Quests/PersistenceData_QuestDevice.PersistenceData_QuestDevice_C|GetbIsDataLoaded _persistencedataobject)
        (return true _persistencedataobject)
        (else
          (return false 0))))))

(fn PlayerTeamApplies
  (params (Actor :type Actor%20Object%20Reference))
  (bind (_teamtype _teamindex) (Utilities|Struct|BreakCreativeTeamOption (Variables|UserOptions|GetValidTeam)))
  (bind (_teamindex_2 _returnvalue) (Team|GetActorPvPTeamIndex Actor))
  (return (or (and _returnvalue (== _teamindex_2 (Knobs|Team|GetTeamValueasIndex _teamindex))) (Utilities|Enum|Equal(Enum) _teamtype "Any"))))

(fn PopulatePersistenceDataObject
  (params (Controller :type Controller%20Object%20Reference))
  (Persistence|ResetPersistenceObject)
  (Utilities|IsValid Controller
    (:Is%20Valid
      (Variables|/Game/Creative/Devices/Quests/PersistenceData_QuestDevice.PersistenceData_QuestDevice_C|SetbActive (Class|CreativeQuestComponent|IsActiveFor (Variables|Default|GetCreativeQuest) (Variables|GetController)) (Variables|Persistence|GetPersistenceDataObject))
      (Variables|/Game/Creative/Devices/Quests/PersistenceData_QuestDevice.PersistenceData_QuestDevice_C|SetProgress (Default|GetCurrentAmount (Variables|GetController)) (Variables|Persistence|GetPersistenceDataObject))
      (Variables|/Game/Creative/Devices/Quests/PersistenceData_QuestDevice.PersistenceData_QuestDevice_C|SetTarget (Variables|Default|GetMinigameTargetAmount) (Variables|Persistence|GetPersistenceDataObject))
      (Variables|/Game/Creative/Devices/Quests/PersistenceData_QuestDevice.PersistenceData_QuestDevice_C|SetbIsDataLoaded true (Variables|Persistence|GetPersistenceDataObject)))))

(fn ResetPersistenceObject
  (bind _persistencedataobject (Variables|Persistence|GetPersistenceDataObject))
  (Variables|/Game/Creative/Devices/Quests/PersistenceData_QuestDevice.PersistenceData_QuestDevice_C|SetbActive false _persistencedataobject)
  (Variables|/Game/Creative/Devices/Quests/PersistenceData_QuestDevice.PersistenceData_QuestDevice_C|SetProgress 0 _persistencedataobject)
  (Variables|/Game/Creative/Devices/Quests/PersistenceData_QuestDevice.PersistenceData_QuestDevice_C|SetTarget 0 _persistencedataobject)
  (Variables|/Game/Creative/Devices/Quests/PersistenceData_QuestDevice.PersistenceData_QuestDevice_C|SetbIsDataLoaded false _persistencedataobject))

(fn ResetDistanceTimer
  (bind _creativequest (Variables|Default|GetCreativeQuest))
  (if (Class|CreativeQuestComponent|IsTrackingDistance _creativequest)
    (Class|CreativeQuestComponent|OnMinigameEnded _creativequest) ; Stop the timer if it's running
    (Class|CreativeQuestComponent|OnMinigameStarted _creativequest) ; Start the timer
  ))

(fn SaveForPlayer
  (params (Controller :type Controller%20Object%20Reference))
  (bind _controller (Variables|GetController))
  (Utilities|IsValid Controller
    (:Is%20Valid
      (Persistence|PopulatePersistenceDataObject Controller)
      (Persistence|SaveData (Variables|Default|GetFortCreativePersistenceComponent) (Variables|Persistence|GetPersistenceDataObject) (PlayerState|GetUniqueNetId (Variables|/Script/Engine.Controller|GetPlayerState _controller)) (Variables|Persistence|GetSavePath) false)
      (GameplayEvent|BroadcastEvent (Variables|UserOptions|Events|GetOn%20Saved) 0 (Utilities|Struct|MakeGameplayEventonSaved _controller)))))

(fn SaveProgress
  (params (Instigator :type Controller%20Object%20Reference%20%28by%20ref%29) (bSaveForAllIndividuals :type Boolean))
  (if (and bSaveForAllIndividuals (Utilities|Enum|NotEqual(Enum) (Variables|UserOptions|GetSharing) "All"))
    (bind _playerstate (Utilities|ForEachPersistentPlayer)
      (:LoopBody
        (Persistence|SaveforPlayer (CreativeAI|Controller|GetParticipantControllerfromActor _playerstate false))))
    (else
      (for _array_element (Class|CreativeQuestComponent|GetActiveSharedControllers (Variables|Default|GetCreativeQuest) (Variables|Actor|GetInstigator))
        (Persistence|SaveforPlayer _array_element)))))

(fn SetCurrentAmount
  (params (Controller :type Controller%20Object%20Reference) (Amount :type Integer))
  (bind _minigamestats_component (LocalVariables|GetMinigameStats_component))
  (Class|CreativeQuestComponent|SetCurrentAmount (Variables|Default|GetCreativeQuest) Controller (Math|Integer|SelectInt (Math|Integer|Max(Integer) Amount) (Math|Integer|Clamp(Integer) Amount 0 (Variables|Default|GetMinigameTargetAmount)) (or (Utilities|Enum|Equal(Enum) (Variables|UserOptions|GetWhenTargetIsReached) "NewEnumerator1") (== (Variables|Default|GetMinigameTargetAmount) 0))))
  (Default|CheckTargetAmountReached (Variables|GetController))
  (if (Variables|CustomStats-ToImplementLater|GetUse%20Custom%20Stat)
    (bind _output_get (LocalVariables|SetMinigameStats_component (Minigame|Stats|GetMinigameStatsComponent self)))
    (Utilities|IsValid _output_get
      (:Is%20Valid
        (bind _controller (Variables|GetController)
          (:then
            (bind _playerstate (Variables|/Script/Engine.Controller|GetPlayerState _controller)
              (:then
                (bind _returnvalue (Minigame|Stats|GetStatfromName _minigamestats_component (Variables|CustomStats-ToImplementLater|GetCustom%20Stat%20To%20Track)))
                (Minigame|Stats|SetStatforPlayerId _minigamestats_component (PlayerState|GetUniqueNetId _playerstate) _returnvalue (Variables|GetAmount) false))
              (:else)))
          (:else))))
    (else
      ;[1 0.1458 0.1458] Jira FORT-450197 - Make sure the minigame stat component also updates when setting score.
      (if (and (Utilities|Enum|Equal(Enum) (Variables|UserOptions|Basic|GetStatToTrack) "Score") (Variables|UserOptions|GetbOverrideScoreboard))
        (bind _output_get_2 (LocalVariables|SetMinigameStats_component (Minigame|Stats|GetMinigameStatsComponent self)))
        (Utilities|IsValid _output_get_2
          (:Is%20Valid
            (Minigame|Stats|SetStatforPlayerId (LocalVariables|GetMinigameStats_component) (PlayerState|GetUniqueNetId (Variables|/Script/Engine.Controller|GetPlayerState (Variables|GetController))) (Class|FortMinigameStatsComponent|GetLoadedStatfromClass (LocalVariables|GetMinigameStats_component) (LocalVariables|GetMinigameStat_score)) (Variables|GetAmount) false)))))))

(fn SetDescriptionText
  (params (Text :type Localizable%20Message%20Structure))
  (Networking|SwitchHasAuthority
    (:Authority
      (Class|FortLocalizableMessageComponent|LocalizeandSetMessage (Variables|Default|GetFortLocalizableMessage) (Variables|GetText) "DescriptionTextOverride" "DescriptionText"))))

(fn SetMinigameTargetAmount
  (params (Amount :type Integer) (Instigator :type Controller%20Object%20Reference))
  (bind _output_get (Variables|Default|SetMinigameTargetAmount Amount))
  (Variables|/Script/FortniteGame.CreativeQuestComponent|SetMinigameTargetAmount _output_get (Variables|Default|GetCreativeQuest))
  ;[1 1 1] FORT-438964: Because we changed the target amount, check if any player who has this quest has completed it (not just the instigator)
  (bind _currentminigame (Variables|/Script/FortniteGame.FortMinigameLogicComponent|GetCurrentMinigame (Variables|Default|GetFortMinigameProgress))
    (:then
      (for _array_element (Minigame|GetParticipatingPlayers _currentminigame)
        (Utilities|QuestIsActive (CreativeAI|Controller|GetParticipantControllerfromActor _array_element false))
        (Default|CheckTargetAmountReached (CreativeAI|Controller|GetParticipantControllerfromActor _array_element false))))
    (:else)))

(fn SetTargetValue
  (params (TargetValue :type Integer))
  (bind _returnvalue (Math|Integer|Clamp(Integer) TargetValue 0 10000))
  (bind _minigametargetamount (Variables|Default|GetMinigameTargetAmount))
  (bind _returnvalue_2 (- _returnvalue _minigametargetamount))
  (Networking|SwitchHasAuthority
    (:Authority
      (if (!= _returnvalue_2 0)
        (Default|IncrementMinigameTargetAmount _returnvalue_2)
        (return)
        (else
          (return))))))

(fn SetTitleText
  (params (Text :type Localizable%20Message%20Structure))
  (Networking|SwitchHasAuthority
    (:Authority
      (Class|FortLocalizableMessageComponent|LocalizeandSetMessage (Variables|Default|GetFortLocalizableMessage) (Variables|GetText) "TitleTextOverride" "TrackerTitle"))))

(fn SetValue
  (params (Player :type Controller%20Object%20Reference) (Value :type Integer))
  (Utilities|QuestIsActive Player)
  (Default|SetCurrentAmount Player Value))

(fn SetValueForAll
  (params (Value :type Integer))
  (bind _currentminigame (Variables|/Script/FortniteGame.FortMinigameLogicComponent|GetCurrentMinigame (Variables|Default|GetFortMinigameProgress))
    (:then
      (bind _outplayers (Minigame|GetParticipatingPlayers_NotPure _currentminigame))
      (for _array_element _outplayers
        (Utilities|IsValid _array_element
          (:Is%20Valid
            (Utilities|QuestIsActive (CreativeAI|Controller|GetParticipantControllerfromActor _array_element false))
            (Default|SetCurrentAmount (CreativeAI|Controller|GetParticipantControllerfromActor _array_element false) (Variables|GetValue))
            ;[0 0.7661 1 0.2] If everyone shares the same progress, we only need to set for one user and it'll set for them all, so break out of the function once we've successfully set one of the values.
            (if (Utilities|Enum|Equal(Enum) (Variables|UserOptions|GetSharing) "All")
              (return))))))
    (:else)))

(fn SetValueForTeam
  (params (TeamIndex :type Integer) (Value :type Integer))
  (if (Utilities|Enum|NotEqual(Enum) (Variables|UserOptions|GetSharing) "All")
    (bind _currentminigame (Variables|/Script/FortniteGame.FortMinigameLogicComponent|GetCurrentMinigame (Variables|Default|GetFortMinigameProgress))
      (:then
        (Utilities|IsValid (Minigame|GetMinigameTeamComponent _currentminigame)
          (:Is%20Valid
            (bind _result (ArrayMacros|GetValidIndex (Class|FortMinigameTeamComponent|GetTeamMembers (Minigame|GetMinigameTeamComponent _currentminigame) (Teams|IndextoTeam (- TeamIndex 1))))
              (:Valid%20Index
                (Utilities|IsValid _result
                  (:Is%20Valid
                    (Utilities|QuestIsActive (CreativeAI|Controller|GetParticipantControllerfromActor _result false))
                    (Default|SetCurrentAmount (CreativeAI|Controller|GetParticipantControllerfromActor _result false) Value)
                    ;[0 0.7661 1 0.2] If everyone on the team shares the same progress, we only need to set for one user and it'll set for them all, so break out of the function once we've successfully set one of the values.
                    (if (Utilities|Enum|Equal(Enum) (Variables|UserOptions|GetSharing) "Team")
                      (return)))))))))
      (:else))
    (else
      (Verse|SetValueforAll Value))))

(fn Unassign
  (params (ForPlayer :type Controller%20Object%20Reference))
  (Class|CreativeQuestComponent|SetActiveFor (Variables|Default|GetCreativeQuest) ForPlayer false))

(fn UpdateGameState
  (bind _creativequest (Variables|Default|GetCreativeQuest))
  (bind _fortminigameprogress (Variables|Default|GetFortMinigameProgress)
    (:then
      (bind _currentminigame (Variables|/Script/FortniteGame.FortMinigameLogicComponent|GetCurrentMinigame _fortminigameprogress)
        (:then
          (if (Utilities|Enum|Equal(Enum) (Variables|/Script/FortniteGame.FortMinigame|GetCurrentState _currentminigame) "PreGame")
            (Class|CreativeQuestComponent|UnregisterQuest _creativequest)
            (else
              (Class|CreativeQuestComponent|RegisterQuest _creativequest _currentminigame))))
        (:else)))
    (:else)))

(fn UpdateQuestAndObjectiveText
  (Utilities|FlowControl|Sequence
    (:then_0
      (Utilities|IsValid (Variables|Default|GetCreativeQuest)
        (:Is%20Valid
          (Variables|/Script/FortniteGame.CreativeQuestComponent|SetQuestName (Utilities|String|ToString(Text) (select (Utilities|Text|TextIsEmpty (Variables|Verse|GetTitleTextOverride)) (Variables|UserOptions|Basic|GetTrackerTitle) (Variables|Verse|GetTitleTextOverride))) (Variables|Default|GetCreativeQuest)))))
    (:then_1
      (Utilities|IsValid (Variables|Default|GetCreativeQuest)
        (:Is%20Valid
          (Variables|/Script/FortniteGame.CreativeQuestComponent|SetObjectiveText (Utilities|String|ToString(Text) (select (Utilities|Text|TextIsEmpty (Variables|Verse|GetDescriptionTextOverride)) (Variables|UserOptions|Basic|GetDescriptionText) (Variables|Verse|GetDescriptionTextOverride))) (Variables|Default|GetCreativeQuest)))))
    (:then_2
      (Networking|CanExecuteCosmeticEvents
        (:True
          (Class|CreativeQuestComponent|TriggerHUDUpdate (Variables|Default|GetCreativeQuest) true))))))

(fn ConstructionScript)

(fn WinningTeamType
  (params (Instigator :type Controller%20Object%20Reference))
  (bind (_teamtype _teamindex) (Utilities|Struct|BreakCreativeTeamOption (Variables|UserOptions|GetWinningTeam)))
  (bind (_teamtype_2 _teamindex_2) (Utilities|Struct|BreakCreativeTeamOption (Variables|UserOptions|GetWinningTeam)))
  (Utilities|IsValid Instigator
    (:Is%20Valid
      (LocalVariables|SetInstigatorTeam (Team|GetActorPvPTeamIndex Instigator)))
    (:Is%20Not%20Valid
      _))
  (return (select (and (Utilities|IsValid (Variables|Actor|GetInstigator) :!class /Script/Engine.KismetSystemLibrary) (Utilities|Enum|Equal(Enum) _teamtype "CompletingTeamWins")) "NewEnumerator1" (select (Knobs|Team|IsValidTeamIndex _teamindex) "NewEnumerator2" "NewEnumerator0")) (Utilities|Select :Index (select (and (Utilities|IsValid (Variables|Actor|GetInstigator) :!class /Script/Engine.KismetSystemLibrary) (Utilities|Enum|Equal(Enum) _teamtype "CompletingTeamWins")) "NewEnumerator1" (select (Knobs|Team|IsValidTeamIndex _teamindex) "NewEnumerator2" "NewEnumerator0")) :NewEnumerator0 0 :NewEnumerator1 (LocalVariables|GetInstigatorTeam) :NewEnumerator2 (Math|Conversions|ToByte(Integer) (Knobs|Team|GetTeamValueasIndex _teamindex_2)))))
