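; Fortnite Creative Trick Tile device that makes buildings in its zone vanish and lose collision when triggered by player contact or channel events.
; A gameplay volume collects target buildings, and enabled and user-restriction components gate triggering.
; A replicated tile-state enum drives delayed activation, destruction and restoration, with optional timed reset.
; Clients show staged glow and bounce during the delay, caching per-building dynamic materials to avoid shared edits.
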(event ChildFunctions|EventPlayModeChanged
  (params (Minigame :type Fort%20Minigame%20Object%20Reference) (bIsInPlayMode :type Boolean))
  ;[0 1 0.1052 0.5] Initialization
  (|CallParentFunction Minigame bIsInPlayMode)
  (Networking|SwitchHasAuthority
    (:Authority
      (if bIsInPlayMode
        (CallFunction|CollectTargetAssets)))
    (:Remote
      (bind _astrick_tile_gameplay_volume (Utilities|Casting|CastToTrickTileGameplayVolume (Variables|/Script/Engine.ChildActorComponent|GetChildActor (Variables|Default|GetTrickTileGameplayVolume))))
      (Class|TrickTileGameplayVolume|ChangeVisibility _astrick_tile_gameplay_volume (not bIsInPlayMode)))))

(event EventBeginPlay
  ;[0 1 0.1052 0.5] Initialization
  (|CallParentFunction)
  (Networking|SwitchHasAuthority
    (:Authority
      (CallFunction|InitializeDevice))
    (:Remote
      (TileState|OnTileStateChange))))

(event EventInitializeDevice
  ;[0 1 0.1052 0.5] Initialization
  (|CallParentFunction)
  (CallFunction|MigrateSameChannelInstances)
  (CallFunction|InitializeComponents)
  (CallFunction|UpdateScaleBasedOnSettings)
  (Networking|CanExecuteCosmeticEvents
    (:True
      (VFX|UpdateActivationVFXTimes)
      (bind _astrick_tile_gameplay_volume (Utilities|Casting|CastToTrickTileGameplayVolume (Variables|/Script/Engine.ChildActorComponent|GetChildActor (Variables|Default|GetTrickTileGameplayVolume))))
      (Class|TrickTileGameplayVolume|ChangeVisibility _astrick_tile_gameplay_volume false))))

(event OnAnyPropertyChanged(ToyOptionsComponent)
  ;[0 1 0.1052 0.5] Initialization
  (CallFunction|InitializeDevice))

(event OnPropertiesLoaded(ToyOptionsComponent)
  ;[0 1 0.1052 0.5] Initialization
  (CallFunction|InitializeDevice))

(event EventDestroyed
  ;[1 0.04377 0] Destruction
  (if (Utilities|Enum|Equal(Enum) (Variables|Replicated|GetForcedTileState) "NewEnumerator3")
    (Utilities|Time|ClearandInvalidateTimerbyHandle (Variables|Default|GetActivationTimerHandler))
    (Utilities|Time|ClearandInvalidateTimerbyHandle (Variables|Default|GetWaitingForDestroyAnimationFinishHandle))
    (Utilities|Time|ClearandInvalidateTimerbyHandle (Variables|Default|GetResetAfterTimerHandle))))

(event OnComponentBeginOverlap(TriggerBox)
  (params (OverlappedComponent :type Primitive%20Component%20Object%20Reference) (OtherActor :type Actor%20Object%20Reference) (OtherComp :type Primitive%20Component%20Object%20Reference) (OtherBodyIndex :type Integer) (bFromSweep :type Boolean) (SweepResult :type Hit%20Result%20Structure%20%28by%20ref%29))
  ;[1 1 1] Handle Overlap
  (Utilities|IsAffectingBuildings
    (:True
      (Utilities|IsValid OtherActor
        (:Is%20Valid
          (if (Trigger|IsActorValidtoTrigger OtherActor)
            (bind (_valid_controller _fortplayerpawn _controller) (CreativeUserRestrictions|ActorHasValidParticipantController (Variables|Default|GetCreative_UseRestriction_Component) OtherActor true true true))
            (if _valid_controller
              (Trigger|TryTriggerTrap _controller))))))))

(event Custom|Trigger%20Event
  (params (TriggerInstigator :type Controller%20Object%20Reference))
  ;[0.5229 0 1 0.5] Receivers
  (Trigger|TryTriggerTrap 0))

(event PawnNowHasValidParticipantController(Creative_UseRestriction_Component)
  (params (FortPlayerPawn :type Fort%20Player%20Pawn%20Object%20Reference) (Controller :type Controller%20Object%20Reference))
  ;[1 1 1] Handle Overlap
  (if (Collision|IsOverlappingActor (Variables|Default|GetTriggerBox) FortPlayerPawn :!class /Script/Engine.PrimitiveComponent)
    (Trigger|TryTriggerTrap Controller)))

(event Custom|Reset%20Event
  (params (Instigator :type Controller%20Object%20Reference))
  ;[0.5229 0 1 0.5] Receivers
  (CallFunction|TryResetTile))

(event Custom|Enable%20Event
  (params (Instigator :type Controller%20Object%20Reference))
  ;[0.5229 0 1 0.5] Receivers
  (Class|CreativeEnabledComponent|SetEnabled (Variables|Default|GetCreative_Enabled_Component) true))

(event Custom|Collect%20Target%20Assets
  (CallFunction|CacheTrackedActors))

(event Custom|Disable%20Event
  (params (Instigator :type Controller%20Object%20Reference))
  ;[0.5229 0 1 0.5] Receivers
  (Class|CreativeEnabledComponent|SetEnabled (Variables|Default|GetCreative_Enabled_Component) false))

(event Custom|Toggle%20Event
  (params (Instigator :type Controller%20Object%20Reference))
  ;[0.5229 0 1 0.5] Receivers
  (Class|CreativeEnabledComponent|SetEnabled (Variables|Default|GetCreative_Enabled_Component) (not (Class|CreativeEnabledComponent|IsEnabled (Variables|Default|GetCreative_Enabled_Component)))))

(event Custom|Enable%20Player%20Contact
  (params (Instigator :type Controller%20Object%20Reference))
  ;[0.5229 0 1 0.5] Receivers
  (if (not (Variables|Replicated|GetDesiredTriggerState))
    (Variables|Replicated|SetDesiredTriggerState true)
    (Trigger|ServerUpdateTrigger)))

(event Custom|Disable%20Player%20Contact
  (params (Instigator :type Controller%20Object%20Reference))
  ;[0.5229 0 1 0.5] Receivers
  (if (Variables|Replicated|GetDesiredTriggerState)
    (Variables|Replicated|SetDesiredTriggerState false)
    (Trigger|ServerUpdateTrigger)))

(event Custom|Toggle%20Player%20Contect
  (params (Instigator :type Controller%20Object%20Reference))
  ;[0.5229 0 1 0.5] Receivers
  (Variables|Replicated|SetDesiredTriggerState (not (Variables|Replicated|GetDesiredTriggerState)))
  (Trigger|ServerUpdateTrigger))

(event MinigameStartedDelegate(FortMinigameProgress)
  ;[1 1 1] Handle Minigame events
  (Networking|SwitchHasAuthority
    (:Authority
      (Utilities|IsinProgress
        (:Yes
          (Variables|Replicated|SetDesiredTriggerState (Variables|UserOptions|Basic|GetTrigger%20on%20Player%20Contact))
          (Trigger|ServerUpdateTrigger)
          (if (Variables|UserOptions|GetActivate%20At%20Game%20Start)
            (TileState|SetForcedTileState "NewEnumerator2")))) ; This really shouldn't be necessary, but for some reason\nthis delegate is firing on MinigameEnd.
    )))

(event MinigameEndedDelegate(FortMinigameProgress)
  ;[1 1 1] Handle Minigame events
  (Utilities|Time|ClearandInvalidateTimerbyHandle (Variables|Default|GetResetAfterTimerHandle))
  (TileState|RestoreTiletoInitialState))

(event Custom|OnDestroyAnimationFinish
  (CallFunction|SetDestroyedVisibility)
  (Utilities|Time|InvalidateTimerHandle (Variables|Default|GetWaitingForDestroyAnimationFinishHandle)))

(event Custom|StartVFXTimeline
  (bind _activation_stage (Variables|VFX|GetActivation%20Stage))
  (bind _vfxdataperstage (Variables|VFX|GetVFXDataPerStage))
  (bind (_output_glowcolor_8_8eb4076e423a62fecc95eda562d126d1 _output_glowfrequency_11_86f5d51e4a81205b5bc1f28583856721) (Utilities|Struct|BreakStructTrickTileVFXData (Utilities|Array|Get(aref) _vfxdataperstage _activation_stage)))
  (bind (_output_glowcolor_8_8eb4076e423a62fecc95eda562d126d1_r _output_glowcolor_8_8eb4076e423a62fecc95eda562d126d1_g _output_glowcolor_8_8eb4076e423a62fecc95eda562d126d1_b _output_glowcolor_8_8eb4076e423a62fecc95eda562d126d1_a) (Math|Color|BreakColor _output_glowcolor_8_8eb4076e423a62fecc95eda562d126d1))
  (bind _glowintensity (CallFunction|GetGlowIntensitybyCurrentActivationTime _output_glowcolor_8_8eb4076e423a62fecc95eda562d126d1_a _output_glowfrequency_11_86f5d51e4a81205b5bc1f28583856721))
  (bind (_returnvalue_r _returnvalue_g _returnvalue_b _returnvalue_a) (Math|Color|BreakColor (Math|Color|LinearColor*Float (Math|Color|MakeColor _output_glowcolor_8_8eb4076e423a62fecc95eda562d126d1_r _output_glowcolor_8_8eb4076e423a62fecc95eda562d126d1_g _output_glowcolor_8_8eb4076e423a62fecc95eda562d126d1_b _output_glowcolor_8_8eb4076e423a62fecc95eda562d126d1_a) _glowintensity)))
  (bind (_output_glowcolor_8_8eb4076e423a62fecc95eda562d126d1_2 _output_glowfrequency_11_86f5d51e4a81205b5bc1f28583856721_2 _output_bounceintensity_5_3e346d374b936ec30283a490ff3172c7 _output_bouncefrequency_7_15970f9041dde520322e1bafc89fba97) (Utilities|Struct|BreakStructTrickTileVFXData (Utilities|Array|Get(aref) (Variables|VFX|GetVFXDataPerStage) (Variables|VFX|GetActivation%20Stage))))
  (CallFunction|CreateMIDsForAffectedBuildings)
  (Timeline|VFXTimeline PlayFromStart
    (:Update
      (Variables|VFX|SetCurrentActivationTime (Math|Float|Min(Float) (* (Components|Timeline|GetPlaybackPosition (Variables|Device_TrickTile_V2|GetVFXTimeline)) (Variables|UserOptions|Basic|GetActivation%20Delay)) (Variables|UserOptions|Basic|GetActivation%20Delay)))
      (Utilities|FlowControl|Sequence
        (:then_0
          (CallFunction|SetActivationStatebyCurrentActivationTime))
        (:then_1
          (if (or (Utilities|Enum|Equal(Enum) (Variables|UserOptions|GetTriggerFeedback) "NewEnumerator0") (Utilities|Enum|Equal(Enum) (Variables|UserOptions|GetTriggerFeedback) "NewEnumerator2"))
            (for _array_element (Variables|Default|GetAffected%20Buildings)
              (if (CallFunction|CanDoBuildingAnimations _array_element)
                (Rendering|Material|SetCustomPrimitiveDataVector4 (Building|GetBuildingMeshComponent _array_element) 7 (Math|Vector4|MakeVector4 _returnvalue_r _returnvalue_g _returnvalue_b _returnvalue_a))
                (else
                  (for _array_element_2 (Rendering|Material|GetMaterials (Building|GetBuildingMeshComponent _array_element))
                    (bind _asmaterial_instance_dynamic (Utilities|Casting|CastToMaterialInstanceDynamic _array_element_2))
                    (Rendering|Material|SetVectorParameterValue _asmaterial_instance_dynamic "BlueprintColor" (Math|Color|MakeColor _returnvalue_r _returnvalue_g _returnvalue_b _returnvalue_a) :!class /Script/Engine.MaterialInstanceDynamic)))))))
        (:then_2
          (if (or (Utilities|Enum|Equal(Enum) (Variables|UserOptions|GetTriggerFeedback) "NewEnumerator0") (Utilities|Enum|Equal(Enum) (Variables|UserOptions|GetTriggerFeedback) "NewEnumerator1"))
            (if (!= (Variables|VFX|GetHitAnimTimer) -1.0)
              (bind _output_get (Variables|VFX|SetHitAnimTimer (+ (Utilities|Time|GetWorldDeltaSeconds) (Variables|VFX|GetHitAnimTimer))))
              (if (>= _output_get _output_bouncefrequency_7_15970f9041dde520322e1bafc89fba97)
                (VFX|PlayCustomHitBounceAnimation)))))))))

(event Custom|StopVFXTimeline
  (bind _activation_stage (Variables|VFX|GetActivation%20Stage))
  (bind _vfxdataperstage (Variables|VFX|GetVFXDataPerStage))
  (bind (_output_glowcolor_8_8eb4076e423a62fecc95eda562d126d1 _output_glowfrequency_11_86f5d51e4a81205b5bc1f28583856721) (Utilities|Struct|BreakStructTrickTileVFXData (Utilities|Array|Get(aref) _vfxdataperstage _activation_stage)))
  (bind (_output_glowcolor_8_8eb4076e423a62fecc95eda562d126d1_r _output_glowcolor_8_8eb4076e423a62fecc95eda562d126d1_g _output_glowcolor_8_8eb4076e423a62fecc95eda562d126d1_b _output_glowcolor_8_8eb4076e423a62fecc95eda562d126d1_a) (Math|Color|BreakColor _output_glowcolor_8_8eb4076e423a62fecc95eda562d126d1))
  (bind _glowintensity (CallFunction|GetGlowIntensitybyCurrentActivationTime _output_glowcolor_8_8eb4076e423a62fecc95eda562d126d1_a _output_glowfrequency_11_86f5d51e4a81205b5bc1f28583856721))
  (bind (_returnvalue_r _returnvalue_g _returnvalue_b _returnvalue_a) (Math|Color|BreakColor (Math|Color|LinearColor*Float (Math|Color|MakeColor _output_glowcolor_8_8eb4076e423a62fecc95eda562d126d1_r _output_glowcolor_8_8eb4076e423a62fecc95eda562d126d1_g _output_glowcolor_8_8eb4076e423a62fecc95eda562d126d1_b _output_glowcolor_8_8eb4076e423a62fecc95eda562d126d1_a) _glowintensity)))
  (bind (_output_glowcolor_8_8eb4076e423a62fecc95eda562d126d1_2 _output_glowfrequency_11_86f5d51e4a81205b5bc1f28583856721_2 _output_bounceintensity_5_3e346d374b936ec30283a490ff3172c7 _output_bouncefrequency_7_15970f9041dde520322e1bafc89fba97) (Utilities|Struct|BreakStructTrickTileVFXData (Utilities|Array|Get(aref) (Variables|VFX|GetVFXDataPerStage) (Variables|VFX|GetActivation%20Stage))))
  (Timeline|VFXTimeline Stop))

(fn CacheTrackedActors
  (bind _astrick_tile_gameplay_volume (Utilities|Casting|CastToTrickTileGameplayVolume (Variables|/Script/Engine.ChildActorComponent|GetChildActor (Variables|Default|GetTrickTileGameplayVolume))))
  (bind _astrick_tile_playspace (Utilities|Casting|CastToTrickTilePlayspace (Class|GameplayVolume|GetPlayspace _astrick_tile_gameplay_volume)))
  (bind _playspacecomponent_spatialactortracker (Variables|/CreativeCoreDevices/TrickTile/TrickTilePlayspace.TrickTilePlayspace_C|GetPlayspaceComponent_SpatialActorTracker _astrick_tile_playspace)
    (:then
      (bind _volumetricregionnative (Variables|Default|GetVolumetricRegionNative)
        (:then
          (Class|VolumetricRegionNativeComponent|RefreshTrackedActors _volumetricregionnative _playspacecomponent_spatialactortracker)
          (bind _result (Utilities|Set|ToArray (Class|PlayspaceComponentSpatialActorTracker|GetTrackedActors _playspacecomponent_spatialactortracker)))
          (for _array_element _result
            (if (!= _array_element self)
              (bind _asbuilding_static_mesh (Utilities|Casting|CastToBuildingSMActor _array_element))
              (Utilities|Array|Add (LocalVariables|GetLocal%20Tiles) _asbuilding_static_mesh)))
          (Variables|Default|SetAffected%20Buildings (LocalVariables|GetLocal%20Tiles)))
        (:else)))
    (:else)))

(fn CacheOriginalValuesFromBaseTile
  (Utilities|Map|Clear (Variables|Default|GetOriginalHealth))
  (Utilities|Map|Clear (Variables|Default|GetOriginalSuppressHealthBarValue))
  (Utilities|Map|Clear (Variables|Default|GetOriginalCustomMaterialValue))
  (for _array_element (Variables|Default|GetAffected%20Buildings)
    (Utilities|Map|Add (Variables|Default|GetOriginalHealth) _array_element (Health|GetHealth _array_element))
    (Utilities|Map|Add (Variables|Default|GetOriginalSuppressHealthBarValue) _array_element (Variables|/Script/FortniteGame.BuildingActor|GetbSurpressHealthBar _array_element))
    (Utilities|Map|Add (Variables|Default|GetOriginalCustomMaterialValue) _array_element (Variables|/Script/FortniteGame.BuildingSMActor|GetbAllowCustomMaterial _array_element))))

(fn ClearActivation
  (params (bForced :type Boolean))
  (bind _activationtimerhandler (Variables|Default|GetActivationTimerHandler))
  (if bForced
    (Utilities|Time|ClearandInvalidateTimerbyHandle _activationtimerhandler)
    (else
      (Utilities|Time|InvalidateTimerHandle _activationtimerhandler)))
  (Networking|CanExecuteCosmeticEvents
    (:True
      (CallFunction|StopVFXTimeline)
      (if (or (Utilities|Enum|Equal(Enum) (Variables|UserOptions|GetTriggerFeedback) "NewEnumerator0") (Utilities|Enum|Equal(Enum) (Variables|UserOptions|GetTriggerFeedback) "NewEnumerator2"))
        (for _array_element (Variables|Default|GetAffected%20Buildings)
          (if (CallFunction|CanDoBuildingAnimations _array_element)
            (Rendering|Material|SetCustomPrimitiveDataVector4 (Building|GetBuildingMeshComponent _array_element) 7 (0.0 0.0 0.0 0.0))
            (else
              (for _array_element_2 (Rendering|Material|GetMaterials (Building|GetBuildingMeshComponent _array_element))
                (bind _asmaterial_instance_dynamic (Utilities|Casting|CastToMaterialInstanceDynamic _array_element_2))
                (Rendering|Material|SetVectorParameterValue _asmaterial_instance_dynamic "TimestampPayload" (0.0 0.0 0.0 1.0) :!class /Script/Engine.MaterialInstanceDynamic)))))))))

(fn CreateMIDsforAffectedBuildings
  (bind _cachedmids (Variables|Default|GetCachedMIDs))
  ;[0.2086 0.3763 1 0.25] FORT-701792: Before we play the timeline, create a new MID for the base material and replace it to modify it. Otherwise, we will be modifying the same material in ALL objects
  (for _array_element (Variables|Default|GetAffected%20Buildings)
    (bind _output_get (LocalVariables|SetBuilding _array_element))
    (if (.ReturnValue (Utilities|Map|Find _cachedmids _output_get))
      (for (_array_element_2 _array_index) (Utilities|Struct|BreakFortObjectArray (Utilities|Map|Find _cachedmids _output_get))
        (bind _asmaterial_instance_dynamic (Utilities|Casting|CastToMaterialInstanceDynamic _array_element_2))
        (Rendering|Material|SetMaterial (Building|GetBuildingMeshComponent (LocalVariables|GetBuilding)) _array_index _asmaterial_instance_dynamic))
      (else
        (Variables|/Script/FortniteGame.BuildingSMActor|SetbAllowCustomMaterial true (LocalVariables|GetBuilding))
        (bind _returnvalue (CallFunction|ReadTextureData _output_get))
        (LocalVariables|SetBuildingTextureData _returnvalue)
        (Utilities|Array|Clear (LocalVariables|GetMIDsToCache))
        (for (_array_element_3 _array_index_2) (Rendering|Material|GetMaterials (Building|GetBuildingMeshComponent (LocalVariables|GetBuilding)))
          (bind _asmaterial_instance_dynamic_2 (Utilities|Casting|CastToMaterialInstanceDynamic _array_element_3)
            (:then
              (LocalVariables|SetParentMaterial (Variables|/Script/Engine.MaterialInstance|GetParent _asmaterial_instance_dynamic_2)))
            (:CastFailed
              (LocalVariables|SetParentMaterial _array_element_3))) ; If we're already an MID, we want the Parent of the MID, not the actual MID itself.\nCreating an MID from an MID does not work.
          (bind _parentmaterial (LocalVariables|GetParentMaterial)
            (:then
              (bind _returnvalue_2 (Rendering|Material|CreateDynamicMaterialInstance _parentmaterial "None" "Transient" :!class /Script/Engine.KismetMaterialLibrary))
              (Rendering|Material|SetMaterial (Building|GetBuildingMeshComponent (LocalVariables|GetBuilding)) _array_index_2 _returnvalue_2)
              (Utilities|Array|Add (LocalVariables|GetMIDsToCache) _returnvalue_2)
              ;;[0.2086 0.3763 1 0.25] Manually Reapplying Texture Data to the new MID.
              (for (_array_element_4 _array_index_3) (LocalVariables|GetBuildingTextureData)
                (Utilities|IsValid _array_element_4
                  (:Is%20Valid
                    (bind _result (ArrayMacros|GetValidIndex (LocalVariables|GetBuildingParameterNames) _array_index_3)
                      (:Valid%20Index
                        (Rendering|Material|SetTextureParameterValue _returnvalue_2 _result _array_element_4)))))))
            (:else)))
        (Utilities|Map|Add (Variables|Default|GetCachedMIDs) (LocalVariables|GetBuilding) (Utilities|Struct|MakeFortObjectArray (LocalVariables|GetMIDsToCache)))))))

(fn DisableTileCanBeDamaged
  (for _array_element (Variables|Default|GetAffected%20Buildings)
    (if (Variables|/Script/Engine.Actor|GetbCanBeDamaged _array_element)
      (Fortnite|SetCanbeDamaged _array_element)))
  (Variables|Default|SetOverrideCanBeDamaged true))

(fn GetGlowIntensityByCurrentActivationTime
  (params (GlowAlpha :type Float%20%28double-precision%29) (GlowFrequency :type Float%20%28double-precision%29))
  (return (* GlowAlpha (+ (Math|Trig|Sin(Radians) (+ (Math|Float|float*float (* GlowFrequency (Variables|VFX|GetCurrentActivationTime)) (Math|Trig|GetPI) 2.0) (Variables|VFX|GetOffset))) 1.3))))

(fn GetHitNormalForBounce
  (return "0, 0,1.000000"))

(fn GetZoneScale
  ;[1 0.8273 0] Returns the zone scale in Tiles
  (return (Utilities|Select :Index (Variables|UserOptions|GetSize%20Units) :NewEnumerator0 "1.000000,1.000000,1.000000" :NewEnumerator1 (Math|Vector|MakeVector (Variables|UserOptions|GetZone%20Width%20%28Tiles%29) (Variables|UserOptions|GetZone%20Depth%20%28Tiles%29) (Variables|UserOptions|GetZone%20Height%20%28Tiles%29)) :NewEnumerator2 (Math|Vector|MakeVector (/ (Variables|UserOptions|GetZone%20Width%20%28Meters%29) 5.12) (/ (Variables|UserOptions|GetZone%20Depth%20%28Meters%29) 5.12) (/ (Variables|UserOptions|GetZone%20Height%20%28Meters%29) 3.84)))))

(fn InitializeComponents
  (Class|CreativeEnabledComponent|InitializeComponent (Variables|Default|GetCreative_Enabled_Component) (Variables|Default|GetFortMinigameProgress) (Math|Integer|SelectInt 2 -1 (Variables|UserOptions|GetEnabled%20At%20Game%20Start)))
  (Initialization|Initialize (Variables|Default|GetCreative_UseRestriction_Component) (Variables|Default|GetFortMinigameProgress) (Variables|Default|GetCreative_Enabled_Component) 0 (Variables|UserOptions|GetSelected%20Team) (Variables|UserOptions|GetInvert%20Selected%20Team) (Variables|UserOptions|GetSelected%20Class) (Variables|UserOptions|GetInvert%20Selected%20Class) false))

(fn IsVisibleInGame
  (return false))

(fn IsActorValidToTrigger
  (params (ActorTryingToTrigger :type Actor%20Object%20Reference))
  (bind _aspawn (Utilities|Casting|CastToPawn ActorTryingToTrigger)
    (:then
      (Networking|SwitchHasAuthority
        (:Authority
          _)
        (:Remote
          (if (not (Pawn|IsLocallyControlled _aspawn))
            (return false))))
      (return true))
    (:CastFailed
      (return false))))

(fn MigrateTrapTransform
  (bind _returnvalue (Transformation|GetActorRotation))
  (bind _trapmigrationrequired (Variables|LegacyOptions|GetTrapMigrationRequired))
  (bind _trap_migration_transforms (Variables|LegacyOptions|GetTrap%20Migration%20Transforms))
  (bind (_value _returnvalue_2) (Utilities|Map|Find _trap_migration_transforms _trapmigrationrequired))
  (bind (_value_location _value_rotation _value_scale) (Math|Transform|BreakTransform _value))
  ;[1 0 0.6587 0.25] Location of the Prop is different than the location of the original trap, so we need to move it (Child objects will define different 'Trap Migration Required' setting). Manually defining order of operations for the transform.
  (Networking|SwitchHasAuthority
    (:Authority
      (if _returnvalue_2
        (Transformation|SetActorLocation (+ (Transformation|GetActorLocation) (Math|Vector|RotateVector _value_location _returnvalue)))
        (Transformation|SetActorRotation (Math|Rotator|CombineRotators _value_rotation _returnvalue))
        (Transformation|SetActorScale3D (* (Transformation|GetActorScale3D) _value_scale))
        (Variables|LegacyOptions|SetTrapMigrationRequired "NewEnumerator0")
        (CallFunction|Resave)))))

(fn MigrateLegacyUserOptions
  (bind (_safeteamknob _classindex) (Class|GetClassDefaults "/CreativeCoreDevices/Device_TrickTile_V2.Device_TrickTile_V2_C"))
  (bind _resetafterknob (Class|GetClassDefaults "/CreativeCoreDevices/Device_TrickTile_V2.Device_TrickTile_V2_C"))
  (|CallParentFunction)
  (Utilities|FlowControl|Sequence
    (:then_0
      (Utilities|FlowControl|Sequence
        (:then_0
          ;[1 1 1] Splitting Reset into two separate options.
          (if (!= (Variables|LegacyOptions|GetResetAfterKnob) _resetafterknob)
            (Variables|UserOptions|Basic|SetReset%20After%20Use true)
            (Variables|UserOptions|Basic|SetReset%20Time (Math|Conversions|ToFloat(Integer) (Variables|LegacyOptions|GetResetAfterKnob)))
            (Variables|LegacyOptions|SetResetAfterKnob _resetafterknob)
            (LocalVariables|SetMigrated%20Option true)))
        (:then_1
          ;[0 0.2196 1 0.25] Migrating Team and Class. This is a little tricky as Trick Tile was previously using 'Safe Team' so any set team is actually 'Anyone but these guys'
          (Utilities|FlowControl|Sequence
            (:then_0
              (if (!= (Variables|LegacyOptions|GetSafeTeamKnob) _safeteamknob)
                (Variables|UserOptions|SetSelected%20Team (Creative|TeamOption|MakefromLegacyTeamIndex (Variables|LegacyOptions|GetSafeTeamKnob)))
                (Variables|UserOptions|SetInvert%20Selected%20Team true)
                (Variables|LegacyOptions|SetSafeTeamKnob _safeteamknob)
                (LocalVariables|SetMigrated%20Option true)))
            (:then_1
              (if (!= _classindex (Variables|LegacyOptions|GetClassIndex))
                (Variables|UserOptions|SetSelected%20Class (Creative|ClassOption|MakefromLegacyClassIndex (Variables|LegacyOptions|GetClassIndex)))
                (Variables|UserOptions|SetInvert%20Selected%20Class true)
                (Variables|LegacyOptions|SetClassIndex _classindex)
                (LocalVariables|SetMigrated%20Option true)))))))
    (:then_1
      (if (LocalVariables|GetMigrated%20Option)
        (OptionData|RefreshPropertyOverrides (Variables|Default|GetToyOptionsComponent))))))

(fn MigrateSameChannelInstances
  (bind _bwasmigrated (UEFNConversionHelper|MigratetoReceiverwhenSameChannel (Variables|Default|GetEnableWhenReceivingFrom) (Variables|Default|GetDisableWhenReceivingFrom) (Variables|Default|GetToggleWhenReceivingFrom)))
  (bind _bwasmigrated_2 (UEFNConversionHelper|MigratetoReceiverwhenSameChannel (Variables|Default|GetEnablePlayerContactTriggerWhenReceivingFrom) (Variables|Default|GetDisablePlayerContactTriggerWhenReceivingFrom) (Variables|Default|GetTogglePlayerContactTriggerWhenReceivingFrom)))
  (return (and _bwasmigrated_2 _bwasmigrated)))

(fn OnActivationFinish
  (Activation|ClearActivation false)
  (TileState|SetForcedTileState "NewEnumerator2"))

(fn OnRep_AffectedBuildings
  (bind _originalcustommaterialvalue (Variables|Default|GetOriginalCustomMaterialValue))
  (Utilities|Map|Clear (Variables|Default|GetCachedMIDs))
  (bind _keys (Utilities|Map|Keys _originalcustommaterialvalue))
  (for _array_element _keys
    (Variables|/Script/FortniteGame.BuildingSMActor|SetbAllowCustomMaterial (Utilities|Map|Find _originalcustommaterialvalue _array_element) _array_element) ; Reset previous meshes to their original custom material stats \n(in case this moves)
  )
  (CallFunction|CacheOriginalValuesfromBaseTile)
  (Networking|CanExecuteCosmeticEvents
    (:True
      (if (Utilities|Enum|NotEqual(Enum) (Variables|Replicated|GetForcedTileState) "NewEnumerator0")
        (TileState|OnTileStateChange)))))

(fn OnRep_CanTriggerOnPlayerContact
  (Collision|SetCollisionEnabled (Variables|Default|GetTriggerBox) (select (Variables|Replicated|GetCanTriggerOnPlayerContact) "QueryOnly" "NoCollision")))

(fn OnRep_ForcedTileState
  (Networking|SwitchHasAuthority
    (:Remote
      (if (Utilities|Enum|NotEqual(Enum) (Variables|Default|GetOldForcedTileState) (Variables|Replicated|GetForcedTileState))
        (TileState|OnTileStateChange)))))

(fn OnTileStateChange
  (bind _local_tile (LocalVariables|GetLocal%20Tile))
  (bind _local_tile_2 (LocalVariables|GetLocal%20Tile))
  (if (Utilities|Time|IsValidTimerHandle (Variables|Default|GetActivationTimerHandler))
    (Activation|ClearActivation true)
    (else
      _))
  (Utilities|FlowControl|Sequence
    (:then_0
      (Utilities|IsAffectingBuildings
        (:True
          (Networking|SwitchHasAuthority
            (:Authority
              (switch Utilities|FlowControl|Switch|SwitchonEnum_TrickTileState (Variables|Replicated|GetForcedTileState)
                (:NewEnumerator1
                  (CallFunction|SetRestoredCollision)
                  (CallFunction|RestoreDamagetoBaseTile)
                  (Variables|Trigger|SetTrap%20Instigator 0) ; Reset instigator when the tile is restored.
                  (Utilities|Time|SetTimerbyEvent (EventDispatchers|CreateEvent :!event Server%20Update%20Trigger) (Variables|Trigger|GetReset%20Overlap%20Recheck%20Time)) ; After a short delay (to allow for Clients to see the wall \nrebuilding), check to see if anyone is standing on it and \ntrigger if they are.
                  (for _array_element (Variables|Default|GetAffected%20Buildings)
                    (if (CallFunction|CanDoBuildingAnimations _array_element)
                      (CallFunction|ForceVisualState _array_element "EBA_Placement"))))
                (:NewEnumerator2
                  (Default|SetDestroyedCollision)
                  (GameplayEvent|BroadcastEvent (Variables|UserOptions|Events|GetOn%20Activated) 0 (Utilities|Struct|MakeGameplayEventActivated (Variables|Trigger|GetTrap%20Instigator)))
                  (CallFunction|RemoveDamagefromBaseTiles)
                  (if (Variables|UserOptions|Basic|GetReset%20After%20Use)
                    (bind _returnvalue (Utilities|Time|SetTimerbyEvent (EventDispatchers|CreateEvent :!event TryResetTile) (Variables|UserOptions|Basic|GetReset%20Time)))
                    (Variables|Default|SetResetAfterTimerHandle _returnvalue)
                    (else
                      _))
                  (for _array_element_2 (Variables|Default|GetAffected%20Buildings)
                    (LocalVariables|SetLocal%20Tile _array_element_2)
                    (Utilities|FlowControl|Sequence
                      (:then_0
                        (if (Building|IsUnderConstruction _local_tile)
                          (CallFunction|CancelAutoBuild _local_tile)))
                      (:then_1
                        (CallFunction|ForceVisualState _local_tile_2 "EBA_Destruction")
                        (Networking|ForceNetUpdate _local_tile_2)
                        (CallFunction|DisableTileCanbeDamaged)))))
                (:NewEnumerator3
                  (Activation|StartActivation))))
            (:Remote
              (switch Utilities|FlowControl|Switch|SwitchonEnum_TrickTileState (Variables|Replicated|GetForcedTileState)
                (:NewEnumerator1
                  (CallFunction|SetRestoredCollision)
                  (Default|SetRestoredVisibility)
                  (Utilities|Time|ClearandInvalidateTimerbyHandle (Variables|Default|GetWaitingForDestroyAnimationFinishHandle))
                  (for _array_element_3 (Variables|Default|GetAffected%20Buildings)
                    (CallFunction|SetSuppressHealthBar _array_element_3 (Utilities|Map|Find (Variables|Default|GetOriginalSuppressHealthBarValue) _array_element_3))
                    (if (not (CallFunction|CanDoBuildingAnimations _array_element_3))
                      (CallFunction|PlayConstructionBounce _array_element_3))))
                (:NewEnumerator2
                  (Default|SetDestroyedCollision)
                  (for _array_element_4 (Variables|Default|GetAffected%20Buildings)
                    (CallFunction|PlayDestructionEffects (Variables|Default|GetAffected%20Buildings))
                    (if (not (CallFunction|CanDoBuildingAnimations _array_element_4))
                      (CallFunction|SetDestroyedVisibility)
                      (else
                        (LocalVariables|SetAnyNeedsDestructionTimer true))))
                  (if (LocalVariables|GetAnyNeedsDestructionTimer)
                    (bind _returnvalue_2 (Utilities|Time|SetTimerbyEvent (EventDispatchers|CreateEvent :!event OnDestroyAnimationFinish) 2.0))
                    (Variables|Default|SetWaitingForDestroyAnimationFinishHandle _returnvalue_2)
                    (else
                      (Utilities|Time|ClearandInvalidateTimerbyHandle (Variables|Default|GetWaitingForDestroyAnimationFinishHandle)))))
                (:NewEnumerator3
                  (Activation|StartActivation))))))))
    (:then_1
      (Variables|Default|SetOldForcedTileState (Variables|Replicated|GetForcedTileState)))))

(fn PlayCustomHitBounceAnimation
  (bind (_output_glowcolor_8_8eb4076e423a62fecc95eda562d126d1 _output_glowfrequency_11_86f5d51e4a81205b5bc1f28583856721 _output_bounceintensity_5_3e346d374b936ec30283a490ff3172c7) (Utilities|Struct|BreakStructTrickTileVFXData (Utilities|Array|Get(aref) (Variables|VFX|GetVFXDataPerStage) (Variables|VFX|GetActivation%20Stage))))
  (Variables|VFX|SetHitAnimTimer 0.0)
  (for _array_element (Variables|Default|GetAffected%20Buildings)
    (bind (_origin _boxextent _sphereradius) (Collision|GetComponentBounds (Building|GetBuildingMeshComponent _array_element)))
    (bind (_origin_x _origin_y) (Math|Vector|BreakVector _origin))
    (bind (_returnvalue_x _returnvalue_y _returnvalue_z) (Math|Vector|BreakVector (Transformation|GetActorLocation _array_element)))
    (CallFunction|StartBounceAnimation _array_element (Utilities|Time|GetTimeSeconds) (Math|Float|Min(Float) _sphereradius 3e+02) (* (VFX|GetHitNormalforBounce) (* _output_bounceintensity_5_3e346d374b936ec30283a490ff3172c7 (select (CallFunction|CanDoBuildingAnimations _array_element) (Variables|VFX|GetMaxBounceFactorForAnimated) (Variables|VFX|GetMaxBounceFactorForCommon)))) (Math|Vector|MakeVector _origin_x _origin_y (+ _returnvalue_z (select (Utilities|Enum|Equal(Enum) (Variables|Attachment|GetBuildingAttachmentType) "ATTACH_Wall") 128.0 0.0))) "Hit" false)))

(fn RemoveDamageFromBaseTiles
  (for _array_element (Variables|Default|GetAffected%20Buildings)
    (Utilities|Map|Add (Variables|Default|GetOriginalHealth) _array_element (Health|GetHealth _array_element))
    (Health|SetHealth _array_element (Health|GetMaxHealth _array_element))))

(fn RestoreDamageToBaseTile
  (bind _originalhealth (Variables|Default|GetOriginalHealth))
  (for _array_element (Variables|Default|GetAffected%20Buildings)
    (if (.ReturnValue (Utilities|Map|Find _originalhealth _array_element))
      (Health|SetHealth (Utilities|Map|Find _originalhealth _array_element)))))

(fn RestoreTileCanBeDamaged
  (if (Variables|Default|GetOverrideCanBeDamaged)
    (for _array_element (Variables|Default|GetAffected%20Buildings)
      (Fortnite|SetCanbeDamaged _array_element true))
    (Variables|Default|SetOverrideCanBeDamaged false)))

(fn RestoreTileToInitialState
  (if (Utilities|Enum|NotEqual(Enum) (Variables|Replicated|GetForcedTileState) "NewEnumerator0")
    (LocalVariables|SetTileStateBeforeRestore (Variables|Replicated|GetForcedTileState))
    (TileState|SetForcedTileState "NewEnumerator0")
    (Utilities|IsAffectingBuildings
      (:True
        (if (Utilities|Enum|Equal(Enum) (LocalVariables|GetTileStateBeforeRestore) "NewEnumerator2")
          (CallFunction|SetRestoredCollision)
          (Default|SetRestoredVisibility)
          (CallFunction|RestoreTileCanbeDamaged)
          (for _array_element (Variables|Default|GetAffected%20Buildings)
            (if (CallFunction|CanDoBuildingAnimations _array_element)
              (CallFunction|ForceVisualState _array_element "EBA_Placement")))
          (Utilities|Time|ClearandInvalidateTimerbyHandle (Variables|Default|GetWaitingForDestroyAnimationFinishHandle)))))))

(fn ServerUpdateTrigger
  (Networking|SwitchHasAuthority
    (:Authority
      (if (or (Utilities|Enum|Equal(Enum) (Variables|Replicated|GetForcedTileState) "NewEnumerator0") (Utilities|Enum|Equal(Enum) (Variables|Replicated|GetForcedTileState) "NewEnumerator1"))
        (Variables|Replicated|SetCanTriggerOnPlayerContact (Variables|Replicated|GetDesiredTriggerState))))))

(fn SetActivationStateByCurrentActivationTime
  (Utilities|FlowControl|Sequence
    (:then_0
      (if (== (Variables|VFX|GetActivation%20Stage) 0)
        (Math|Float|CompareFloat (Variables|VFX|GetCurrentActivationTime) (Variables|VFX|GetStageOneTime)
          (:>
            (Math|Integer|IncrementInt (Variables|VFX|GetActivation%20Stage))))))
    (:then_1
      (if (== (Variables|VFX|GetActivation%20Stage) 1)
        (Math|Float|CompareFloat (Variables|VFX|GetCurrentActivationTime) (Variables|VFX|GetStageTwoTime)
          (:>
            (Math|Integer|IncrementInt (Variables|VFX|GetActivation%20Stage))))))))

(fn SetDestroyedCollision
  (for _array_element (Variables|Default|GetAffected%20Buildings)
    (bind _astrick_tile_interface (Utilities|Casting|CastToTrickTileInterface _array_element)
      (:then
        (bind _returnvalue (TrickTile|GetStaticMeshComponents _astrick_tile_interface))
        (for _array_element_2 _returnvalue
          (Utilities|IsValid _array_element_2
            (:Is%20Valid
              (Collision|SetCollisionEnabled _array_element_2 "NoCollision")))))
      (:CastFailed
        (Collision|SetCollisionEnabled (Variables|/Script/FortniteGame.BuildingSMActor|GetStaticMeshComponent _array_element) "NoCollision")))))

(fn SetDestroyedVisibility
  (for _array_element (Variables|Default|GetAffected%20Buildings)
    (bind _astrick_tile_interface (Utilities|Casting|CastToTrickTileInterface _array_element)
      (:then
        (bind _returnvalue (TrickTile|GetStaticMeshComponents _astrick_tile_interface))
        (for _array_element_2 _returnvalue
          (Utilities|IsValid _array_element_2
            (:Is%20Valid
              (Rendering|SetVisibility _array_element_2 false false)))))
      (:CastFailed
        (Rendering|SetVisibility (Variables|/Script/FortniteGame.BuildingSMActor|GetStaticMeshComponent _array_element) false false)))))

(fn SetForcedTileState
  (params (NewState :type Enum_TrickTileState%20Enum :default "NewEnumerator0"))
  (if (Utilities|Enum|NotEqual(Enum) NewState (Variables|Replicated|GetForcedTileState))
    (Variables|Replicated|SetForcedTileState NewState)
    (TileState|OnTileStateChange)))

(fn SetRestoredCollision
  (bind _meshcomponent (LocalVariables|GetMeshComponent))
  (bind _defaultcollisions (LocalVariables|GetDefaultCollisions))
  (bind (_value _returnvalue) (Utilities|Map|Find _defaultcollisions _meshcomponent))
  (bind _meshcomponent_2 (LocalVariables|GetMeshComponent))
  (for _array_element (Variables|Default|GetAffected%20Buildings)
    (bind _astrick_tile_interface (Utilities|Casting|CastToTrickTileInterface _array_element)
      (:then
        (bind _returnvalue_2 (TrickTile|GetDefaultCollisions _astrick_tile_interface))
        (LocalVariables|SetDefaultCollisions _returnvalue_2)
        (bind _returnvalue_3 (TrickTile|GetStaticMeshComponents(Message) _array_element))
        (for _array_element_2 _returnvalue_3
          (bind _output_get (LocalVariables|SetMeshComponent _array_element_2))
          (Utilities|IsValid _output_get
            (:Is%20Valid
              (if _returnvalue
                (Collision|SetCollisionEnabled _meshcomponent_2 _value)
                (else
                  (Collision|SetCollisionEnabled _meshcomponent_2 "QueryAndPhysics")))))))
      (:CastFailed
        (Collision|SetCollisionEnabled (Variables|/Script/FortniteGame.BuildingSMActor|GetStaticMeshComponent _array_element) "QueryAndPhysics")))))

(fn SetRestoredVisibility
  (for _array_element (Variables|Default|GetAffected%20Buildings)
    (bind _astrick_tile_interface (Utilities|Casting|CastToTrickTileInterface _array_element)
      (:then
        (bind _returnvalue (TrickTile|GetStaticMeshComponents _astrick_tile_interface))
        (for _array_element_2 _returnvalue
          (Utilities|IsValid _array_element_2
            (:Is%20Valid
              (Rendering|SetVisibility _array_element_2 true false)))))
      (:CastFailed
        (Rendering|SetVisibility (Variables|/Script/FortniteGame.BuildingSMActor|GetStaticMeshComponent _array_element) true false)))))

(fn ShouldTriggerTrap
  (return (Math|Boolean|ANDBoolean (Class|CreativeEnabledComponent|IsEnabled (Variables|Default|GetCreative_Enabled_Component)) (Utilities|Enum|NotEqual(Enum) (Variables|Replicated|GetForcedTileState) "NewEnumerator2") (Utilities|Enum|NotEqual(Enum) (Variables|Replicated|GetForcedTileState) "NewEnumerator3"))))

(fn StartActivation
  (bind _returnvalue (Utilities|Time|SetTimerbyEvent (EventDispatchers|CreateEvent :!event OnActivationFinish) (Variables|UserOptions|Basic|GetActivation%20Delay)))
  (Variables|Default|SetActivationTimerHandler _returnvalue)
  (Networking|SwitchHasAuthority
    (:Authority
      (Variables|Replicated|SetCanTriggerOnPlayerContact false))
    (:Remote
      (if (Utilities|Enum|NotEqual(Enum) (Variables|UserOptions|GetTriggerFeedback) "NewEnumerator3")
        (Variables|VFX|SetActivation%20Stage 0)
        (Variables|VFX|SetCurrentActivationTime 0.0)
        (VFX|StartActivationVFX)))))

(fn StartActivationVFX
  (CallFunction|StartVFXTimeline)
  (Utilities|FlowControl|Sequence
    (:then_0
      (if (or (Utilities|Enum|Equal(Enum) (Variables|UserOptions|GetTriggerFeedback) "NewEnumerator0") (Utilities|Enum|Equal(Enum) (Variables|UserOptions|GetTriggerFeedback) "NewEnumerator2"))
        (for _array_element (Variables|Default|GetAffected%20Buildings)
          (if (not (CallFunction|CanDoBuildingAnimations _array_element))
            (for _array_element_2 (Rendering|Material|GetMaterials (Building|GetBuildingMeshComponent _array_element))
              (bind _asmaterial_instance_dynamic (Utilities|Casting|CastToMaterialInstanceDynamic _array_element_2))
              (Rendering|Material|SetVectorParameterValue _asmaterial_instance_dynamic "TimestampPayload" (Math|Color|MakeColor 0.0 0.0 (+ (Utilities|Time|GetGameTimeinSeconds) 3.6e+04)) :!class /Script/Engine.MaterialInstanceDynamic)
              (Rendering|Material|SetVectorParameterValue _asmaterial_instance_dynamic "BlueprintColor" "(R=0.000000,G=0.000000,B=0.000000,A=1.000000)" :!class /Script/Engine.MaterialInstanceDynamic))))))
    (:then_1
      (if (or (Utilities|Enum|Equal(Enum) (Variables|UserOptions|GetTriggerFeedback) "NewEnumerator0") (Utilities|Enum|Equal(Enum) (Variables|UserOptions|GetTriggerFeedback) "NewEnumerator1"))
        (VFX|PlayCustomHitBounceAnimation)))))

(fn TryResetTile
  (Utilities|Time|ClearandInvalidateTimerbyHandle (Variables|Default|GetResetAfterTimerHandle))
  (if (Math|Byte|Byte>Byte (Variables|Replicated|GetForcedTileState) (Utilities|Enum|LiteralenumEnum_TrickTileState "NewEnumerator1"))
    (TileState|SetForcedTileState "NewEnumerator1")
    (CallFunction|RestoreTileCanbeDamaged)))

(fn TryTriggerTrap
  (params (Attempting%20Controller :type Controller%20Object%20Reference))
  (if (Trigger|ShouldTriggerTrap)
    (Variables|Trigger|SetTrap%20Instigator Attempting%20Controller)
    (if (> (Variables|UserOptions|Basic|GetActivation%20Delay) 0.0)
      (TileState|SetForcedTileState "NewEnumerator3")
      (else
        (TileState|SetForcedTileState "NewEnumerator2")))
    (Networking|SwitchHasAuthority
      (:Authority
        (GameplayEvent|BroadcastEvent (Variables|UserOptions|Events|GetOn%20Triggered) 0 (Utilities|Struct|MakeGameplayEventTriggered (Variables|Trigger|GetTrap%20Instigator)))))))

(fn UpdateScalebasedonSettings
  (bind (_base_scale_min_max_x _base_scale_min_max_y) (Math|Vector2D|BreakVector2D (Variables|GetBase%20Scale%20Min/Max_X)))
  (bind _scale (CallFunction|GetZoneScale))
  ;[1 1 1] Scaling the base mesh and interaction zones.
  (Transformation|SetRelativeScale3D (Variables|Default|GetBaseMeshRoot) (Math|Vector|MakeVector (Math|Float|Clamp(Float) (.x _scale) _base_scale_min_max_x _base_scale_min_max_y) (Math|Float|Clamp(Float) (.y _scale) _base_scale_min_max_x _base_scale_min_max_y) 1.0))
  (Transformation|SetRelativeScale3D (Variables|Default|GetTriggerBox) (Math|Vector|MakeVector (.x _scale) (.y _scale) 1.0))
  (Transformation|SetRelativeScale3D (Variables|Default|GetTrickTileGameplayVolume) (- (* _scale "512.000000,512.000000,384.000000") (Variables|Default|GetGameplay%20Volume%20Offset))))

(fn UpdateActivationVFXTimes
  (if (> (Variables|UserOptions|Basic|GetActivation%20Delay) 0.0)
    (Variables|VFX|SetStageOneTime (- (Variables|UserOptions|Basic|GetActivation%20Delay) (+ (Variables|VFX|GetDesiredStageOneDuration) (Variables|VFX|GetDesiredStageTwoDuration))))
    (Variables|VFX|SetStageTwoTime (- (Variables|UserOptions|Basic|GetActivation%20Delay) (Variables|VFX|GetDesiredStageTwoDuration)))
    (bind _vfxtimeline (Variables|Device_TrickTile_V2|GetVFXTimeline)
      (:then
        (Components|Timeline|SetPlayRate _vfxtimeline (/ 1.0 (Variables|UserOptions|Basic|GetActivation%20Delay))))
      (:else))))

(fn ConstructionScript
  (|CallParentFunction)
  (CallFunction|MigrateTrapTransform))
