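; Parent Blueprint for Creative Trigger devices, providing shared activation, delay, reset and trigger-limit logic for children to extend.
; On authority, activation checks restrictions and eligibility, waits out the delay, broadcasts OnTrigger every Nth hit, then starts the reset.
; CurrentState selects mesh materials, while throttled hits play audio, VFX and a glow using pooled dynamic materials.
; Verse overrides supersede user options until minigame end, and migration functions upgrade legacy options.
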(event OnVisibilityChanged(VisibleInGameComponent)
  (params (Visible :type Boolean))
  ;[1 1 1] Visibility
  (CallFunction|SetMeshVisibilityandCollision Visible))

(event Custom|Load%20and%20Set%20Audio
  ;[1 0.7441 0] Async Load and set the various audio samples for the Trigger
  ;; Moved the Trigger SFX out to always load the sounds, but whether or not they get played is decided on the SendHit event so that we can change them dynamically.
  (Networking|CanExecuteCosmeticEvents
    (:True
      (bind _object (Utilities|AsyncLoadAsset (Variables|SFX|GetTriggerSound))
        (:then
          (bind _object_2 (Utilities|AsyncLoadAsset (Variables|SFX|GetResetSound))
            (:Completed
              (bind _assound_cue (Utilities|Casting|CastToSoundCue _object_2))
              (Audio|Components|Audio|SetSound (Variables|Default|GetReset%20Audio) _assound_cue))))
        (:Completed
          (bind _assound_cue_2 (Utilities|Casting|CastToSoundCue _object))
          (Audio|Components|Audio|SetSound (Variables|Default|GetActivate%20Audio) _assound_cue_2))))))

(event OnAnyPropertyChanged(ToyOptionsComponent)
  ;[0 1 0.3633 0.5] Initialization
  (CallFunction|InitializeDevice))

(event OnPropertiesLoaded(ToyOptionsComponent)
  ;[0 1 0.3633 0.5] Initialization
  (CallFunction|InitializeDevice))

(event Custom|Initialize%20Device
  ;[0 1 0.3633 0.5] Initialization
  (CallFunction|InitializeComponents)
  (CallFunction|EnableOnFirstValidPhase "PreGame") ; Calling here as we may be in pre-game in a featured island or LTM
  (CallFunction|LoadAndSetAudio)
  (CallFunction|ResetFXOverrides))

(event EventEndPlay
  (params (EndPlayReason :type EEndPlayReason%20Enum :default "Destroyed"))
  ;[0 1 0.3633 0.5] Initialization
  (CallFunction|SwitchToOffMaterial))

(event Custom|Hit
  (params (Play%20VFX :type Boolean))
  ;[1 0.8839 0] Cosmetic Effect on Clients
  (Networking|CanExecuteCosmeticEvents
    (:True
      (CallFunction|SwitchToDyamicMaterial)
      (CallFunction|AnimateDynamicMaterialActivate)
      (if (and (Verse|GetShouldPlaySFX) Play%20VFX)
        (Audio|Components|Audio|Play (Variables|Default|GetActivate%20Audio) 0.0)
        (else
          _))
      (if (and (Verse|GetShouldPlayVFX) Play%20VFX)
        (Components|Activation|SetActive (Variables|Default|GetP_Trigger_Activate) true true)))))

(event Custom|Send%20Hit
  (params (Play%20FX :type Boolean))
  ;[1 0.6759 0] Only send to Client Every MinFXDelay seconds
  (Utilities|FlowControl|DoOnce false
    (:Completed
      (CallFunction|Hit Play%20FX)
      (Utilities|FlowControl|Delay (Variables|Default|GetMin%20FX%20Delay)))))

(event Custom|Minigame%20State%20Changed
  (params (New%20State :type EFortMinigameState%20Enum :default "PreGame"))
  ;[0 0.08007 1 0.5] Gameplay State Logic
  (CallFunction|EnableOnFirstValidPhase New%20State))

(event MinigameStateChangedDelegate(FortMinigameLogic)
  (params (Minigame :type Fort%20Minigame%20Object%20Reference) (NewMinigameState :type EFortMinigameState%20Enum :default "PreGame"))
  ;[0 0.08007 1 0.5] Gameplay State Logic
  (CallFunction|MinigameStateChanged NewMinigameState))

(event MinigameEndedDelegate(FortMinigameLogic)
  ;[0 0.08007 1 0.5] Gameplay State Logic
  (Utilities|Time|ClearandInvalidateTimerbyHandle (Variables|Default|GetTriggerDelayTimer))
  (Utilities|Time|ClearandInvalidateTimerbyHandle (Variables|Default|GetResetTimer))
  (CallFunction|DisableEvent)
  (Utilities|FlowControl|Sequence
    (:then_0
      (Variables|Verse|SetTimesCanTriggerOverride -1)
      (Variables|Default|SetTimesCanTriggerOverridden "NewEnumerator0"))
    (:then_1
      (Variables|Verse|SetTriggerDelayOverride -1.0)
      (Variables|Verse|SetResetDelayOverride -1.0))
    (:then_2
      (Variables|Default|SetTriggerVFXOverriddenDuringPlay false)
      (Variables|Default|SetTriggerSFXOverriddenDuringPlay false))))

(event Custom|ResetFXOverrides
  ;[1 0.6759 0] Reset the FX overrides from Verse - We do this to ensure the RepNotify for the Overrides from Verse can be sent to local versions of the Device
  (Variables|Verse|SetTriggerVFXOverride (Variables|UserOptions|GetPlayTriggerVFX))
  (Variables|Verse|SetTriggerSFXOverride (Variables|UserOptions|GetPlayTriggerSFX)))

(event Custom|EndReset
  ;[0 0.432 1] Ending the Reset Period
  (Variables|Default|SetbIsResetting false)
  (if (CallFunction|CanTrigger)
    (Variables|Default|SetCurrentState "NewEnumerator0")
    (else
      (Variables|Default|SetCurrentState "NewEnumerator2"))))

(event Custom|SendActivationEvent
  ;[0 0.3961 1] Actually Activating the Device.
  (Utilities|Time|ClearandInvalidateTimerbyHandle (Variables|Default|GetTriggerDelayTimer))
  (Variables|Default|SetbIsDelayed false)
  (if (Math|Boolean|ANDBoolean (> (Variables|UserOptions|GetTimesCanTrigger) 0) (not (Variables|Default|GetbChildHandlesDecrement)) (Utilities|GetTimesCanTriggerOverride))
    (Variables|Default|SetRemainingTimesCanTrigger (- (Variables|Default|GetRemainingTimesCanTrigger) 1)))
  (CallFunction|AttemptTrigger)
  (Variables|Default|SetTriggerInstigator 0)
  (bind _delay (Verse|GetResetDelay))
  (Variables|Default|SetbIsResetting true)
  (if (> _delay 0.0)
    (bind _returnvalue (Utilities|Time|SetTimerbyEvent (EventDispatchers|CreateEvent :!event Custom|EndReset) 0.01 false false _delay))
    (Variables|Default|SetResetTimer _returnvalue)
    (Variables|Default|SetCurrentState "NewEnumerator2")
    (else
      (CallFunction|EndReset))))

(event Custom|Switch%20to%20Dyamic%20Material
  (bind _material_instance_dynamic (Variables|Materials|GetMaterial%20Instance%20Dynamic))
  ;[1 0.7873 0] Material Updates
  (Networking|CanExecuteCosmeticEvents
    (:True
      (Variables|Materials|GetMaterial%20Instance%20Dynamic
        (:then
          _)
        (:else
          (bind _material_cache (Variables|Materials|GetMaterial%20Cache)
            (:then
              (bind (_outmaterialinstance _returnvalue) (Class|FortMaterialCache|ReserveMaterial _material_cache))
              (Variables|Materials|SetMaterial%20Instance%20Dynamic _outmaterialinstance)
              (if _returnvalue
                _))
            (:else))))
      (Rendering|Material|SetMaterial (Variables|Default|GetMesh) 1 _material_instance_dynamic)
      (Variables|Materials|SetCurrentMaterial _material_instance_dynamic))))

(event Custom|EnableFromEvent
  (params (Instigator :type Controller%20Object%20Reference))
  ;[0.6648 0 1] Receivers
  ;;[1 1 1] Activate/Deactivate when Received goes here
  (CallFunction|EnableEvent) ; Call function to give children something\nto hook into. for custom logic here.
)

(event Custom|Switch%20to%20Animating%20Material
  ;[1 0.7873 0] Material Updates
  (Networking|CanExecuteCosmeticEvents
    (:True
      (CallFunction|SetCurrentConstantMaterial (Variables|Materials|GetDefaultDelayMaterial)))))

(event Custom|Enable%20Event
  ;[0.6648 0 1] Receivers
  (Variables|Default|SetbIsEnabled true)
  ;;[1 1 1] Don't interrupt the current trigger if we are triggering
  (if (or (Variables|Default|GetbIsResetting) (Variables|Default|GetbIsDelayed))
    _
    (elif (CallFunction|CanTrigger)
      (Variables|Default|SetCurrentState "NewEnumerator0"))))

(event Custom|ReleaseDynamicMaterial
  (bind _material_cache (Variables|Materials|GetMaterial%20Cache))
  (bind _material_instance_dynamic (Variables|Materials|GetMaterial%20Instance%20Dynamic))
  ;[1 0.7873 0] Material Updates
  (if (and (Utilities|IsValid _material_cache :!class /Script/Engine.KismetSystemLibrary) (Utilities|IsValid _material_instance_dynamic :!class /Script/Engine.KismetSystemLibrary))
    (Class|FortMaterialCache|ReleaseMaterial _material_cache _material_instance_dynamic)))

(event Custom|Switch%20to%20Off%20Material
  ;[1 0.7873 0] Material Updates
  (Networking|CanExecuteCosmeticEvents
    (:True
      (CallFunction|SetCurrentConstantMaterial (Variables|Materials|GetDefaultOffMaterial)))))

(event Custom|Disable%20Event
  ;[0.6648 0 1] Receivers
  (Variables|Default|SetbIsEnabled false)
  ;;[1 1 1] Don't interrupt the current trigger if we are triggering
  (if (or (Variables|Default|GetbIsResetting) (Variables|Default|GetbIsDelayed))
    _
    (elif (CallFunction|CanTrigger)
      _
      (else
        (Variables|Default|SetCurrentState "NewEnumerator2")))))

(event Custom|DisableFromEvent
  (params (Instigator :type Controller%20Object%20Reference))
  ;[0.6648 0 1] Receivers
  ;;[1 1 1] Activate/Deactivate when Received goes here
  (CallFunction|DisableEvent) ; Call function to give children something\nto hook into. for custom logic here.
)

(event Custom|Switch%20to%20On%20Material
  ;[1 0.7873 0] Material Updates
  (Networking|CanExecuteCosmeticEvents
    (:True
      (CallFunction|SetCurrentConstantMaterial (Variables|Materials|GetDefaultOnMaterial))
      (Audio|Components|Audio|Play (Variables|Default|GetReset%20Audio) 0.0))))

(event Custom|ResetTimesTriggeredFromEvent
  (params (Instigator :type Controller%20Object%20Reference))
  ;[0.6648 0 1] Receivers
  (CallFunction|ResetTriggerCount)
  (if (CallFunction|CanTrigger)
    (Variables|Default|SetCurrentState "NewEnumerator0")))

(event Custom|Animate%20Dynamic%20Material%20Activate
  ;[1 0.7873 0] Material Updates
  (bind _material_instance_dynamic (Variables|Materials|GetMaterial%20Instance%20Dynamic)
    (:then
      (bind (_direction _glow) (Timeline|Triggered PlayFromStart)
        (:Update
          (Rendering|Material|SetScalarParameterValue _material_instance_dynamic "Glow" _glow :!class /Script/Engine.MaterialInstanceDynamic))
        (:Finished
          (Rendering|Material|SetScalarParameterValue _material_instance_dynamic "Glow" _glow :!class /Script/Engine.MaterialInstanceDynamic)
          (Default|SetMaterialBasedonCurrentState))))
    (:else)) ; Skip animating if we failed to get a dynamic\nmaterial instance (too may things animating)
)

(event HandleVersionUpgrades(CreativeVersion)
  (params (LastSavedVersion :type Creative%20Version%20Id%20Structure))
  (CreativeVersion|CompareVersiontoTag LastSavedVersion "(TagName=\"Creative.Version.Release.40.20\")"
    (:OlderVersion
      (Variables|UserOptions|SetPlayTriggerVFX (Variables|Legacy|GetTriggerVFX))
      (Variables|UserOptions|SetPlayTriggerSFX (Variables|Legacy|GetTrigger%20SFX))
      (OptionData|RefreshPropertyOverrides (Variables|Default|GetToyOptionsComponent)))))

(fn ActivateFromOtherDevice
  (params (Instigator :type Controller%20Object%20Reference) (Requester :type ERequesterType%20Enum :default "NewEnumerator0"))
  (return false))

(fn AttemptTrigger
  ;[0 0.7828 1] Attempting Trigger
  (bind _can_trigger (CallFunction|DoNTriggerLogic))
  (CallFunction|SendHit _can_trigger)
  (if _can_trigger
    (CallFunction|TransmitMessage)
    (return)))

(fn BlueprintModifyIncomingDamage
  (params (Damage :type Float%20%28single-precision%29) (InTags :type Gameplay%20Tag%20Container%20Structure%20%28by%20ref%29) (EffectContext :type Gameplay%20Effect%20Context%20Handle%20Structure%20%28by%20ref%29) (EventInstigator :type Controller%20Object%20Reference) (DamageCauser :type Actor%20Object%20Reference))
  (return 0.0))

(fn CanInteract
  (params (InteractingPawn :type Fort%20Pawn%20Object%20Reference) (InteractionBeingAttempted :type EInteractionBeingAttempted%20Enum :default "FirstInteraction") (InteractionType :type TInteractionType%20Enum :default "IT_NoInteraction"))
  (return false))

(fn CanTrigger
  (Utilities|FlowControl|Sequence
    (:then_0
      (LocalVariables|SetbOutCanTrigger (Math|Boolean|ANDBoolean (Class|CreativeCommonDeviceFunctionLibrary|IsinPlayMode self) (not (or (Variables|Default|GetbIsResetting) (Variables|Default|GetbIsDelayed))) (Math|Boolean|ORBoolean (<= (Variables|UserOptions|GetTimesCanTrigger) 0) (> (Variables|Default|GetRemainingTimesCanTrigger) 0) (not (Utilities|GetTimesCanTriggerOverride))) (Variables|Default|GetbIsEnabled))))
    (:then_1
      (return (LocalVariables|GetbOutCanTrigger)))))

(fn DoNTriggerLogic
  (bind _timestriggered (Variables|Default|GetTimesTriggered))
  (bind _transmit_every_x_triggers (CallFunction|TransmitEvery))
  (Math|Integer|IncrementInt _timestriggered
    (:%20%20
      (return (== (Math|Integer|%25%28Integer%29 _timestriggered _transmit_every_x_triggers) 0)))))

(fn EnableonFirstValidPhase
  (params (Current%20State :type EFortMinigameState%20Enum :default "PreGame"))
  (if (or (and (Utilities|Enum|Equal(Enum) Current%20State "PreGame") (Class|CreativeCommonDeviceFunctionLibrary|IsinPlayMode self)) (Utilities|Enum|Equal(Enum) Current%20State "Warmup"))
    (CallFunction|ResetTriggerCount)
    (Variables|Default|SetbIsResetting false)
    (if (Variables|UserOptions|GetEnabled%20On%20Game%20Start)
      (CallFunction|EnableEvent)
      (else
        (CallFunction|DisableEvent)))))

(fn GetDeviceColor
  (return 0))

(fn GetInstigatingController
  (return 0))

(fn GetCollisionStaticMesh
  (return "0, 0, 0" (Variables|/Script/Engine.StaticMeshComponent|GetStaticMesh (Variables|Default|GetMesh))))

(fn GetMaxTriggerCountOverride
  (return (Variables|UserOptions|GetTimesCanTrigger_Override)))

(fn GetMeshComponents
  (return (Utilities|Array|MakeArray (Variables|Default|GetMesh))))

(fn GetMusicSequencerHitSound
  (params (InstigatedBy :type Controller%20Object%20Reference))
  (return 0 0 0 0))

(fn GetMusicSequencerHitSoundLocation
  (params (InstigatedBy :type Controller%20Object%20Reference))
  (return "0, 0, 0"))

(fn GetNumTimesCanTrigger
  (return (Math|Integer|SelectInt (Variables|Verse|GetTimesCanTriggerOverride) (Math|Integer|SelectInt (Variables|UserOptions|GetTimesCanTrigger) 0 (Utilities|GetTimesCanTriggerOverride)) (>= (Variables|Verse|GetTimesCanTriggerOverride) 0))))

(fn GetOverrideMeshMaterial
  (params (MeshComp :type Static%20Mesh%20Component%20Object%20Reference) (MatIdx :type Integer))
  (bind _mesh (Variables|Default|GetMesh))
  (if (Utilities|Equal(Object) (Variables|GetMeshComp) _mesh)
    (return (Utilities|Select :Index (Variables|GetMatIdx) :Option%200 (Rendering|Material|GetMaterial _mesh 0) :Option%201 (Variables|Materials|GetCurrentMaterial)))
    (elif (Utilities|Equal(Object) (Variables|GetMeshComp) (Variables|Default|GetChannelNumberPlane))
      (if (== (Variables|GetMatIdx) 0)
        (return (Variables|Materials|GetChannel%20Material%20Instance))
        (else
          (return 0)))
      (else
        _))))

(fn GetRemainingNumTimesCanTrigger
  (return (Math|Integer|SelectInt (Variables|Default|GetRemainingTimesCanTrigger) 0 (Utilities|GetTimesCanTriggerOverride))))

(fn GetResetDelay
  (return (Math|Float|SelectFloat (Variables|Verse|GetResetDelayOverride) (Variables|UserOptions|GetReset%20Delay) (>= (Variables|Verse|GetResetDelayOverride) 0))))

(fn GetShouldPlaySFX
  (return (select (Variables|Default|GetTriggerSFXOverriddenDuringPlay) (Variables|Verse|GetTriggerSFXOverride) (Variables|UserOptions|GetPlayTriggerSFX))))

(fn GetShouldPlayVFX
  (return (select (Variables|Default|GetTriggerVFXOverriddenDuringPlay) (Variables|Verse|GetTriggerVFXOverride) (Variables|UserOptions|GetPlayTriggerVFX))))

(fn GetSpawnedPickups
  (return 0))

(fn GetTriggerDelay
  (return (Math|Float|SelectFloat (Variables|Verse|GetTriggerDelayOverride) (Variables|UserOptions|GetTriggerDelay) (>= (Variables|Verse|GetTriggerDelayOverride) 0))))

(fn IgnoreWorldGridScaleOverride
  (return false))

(fn InitializeComponents
  (CallFunction|UpdateVisibilityDuringPhase)
  (Initialization|Initialize (Variables|Default|GetCreative_UseRestriction_Component) (Variables|Default|GetFortMinigameLogic) 0 0 (Variables|UserOptions|Basic|GetActivatingTeam) (Variables|UserOptions|GetInvert%20Team%20Selection) (Variables|UserOptions|Basic|GetActivatingClass) (Variables|UserOptions|GetInvert%20Class%20Selection) false))

(fn IsEnabled
  (return false))

(fn IsHoldingItemForPawn
  (params (Item :type Item%20Variant%20Handle%20Structure) (Pawn :type Pawn%20Object%20Reference))
  (return false))

(fn MigrateLegacyUserOptions
  (bind (_timescantrigger _playtriggervfx _playtriggersfx _triggersoundindex _teamsetting _classindex _visibilityingame _enabledonminigamestart _triggerresetdelay) (Class|GetClassDefaults "/CreativeCoreDevices/Device_Trigger_Parent_V2.Device_Trigger_Parent_V2_C"))
  (bind _triggersoundindex_2 (Variables|Legacy|GetTriggerSoundIndex))
  (bind _triggerresetdelay_2 (Variables|Legacy|GetTriggerResetDelay))
  (bind _timescantrigger_2 (Variables|UserOptions|GetTimesCanTrigger))
  (bind _visibilityingame_2 (Variables|Legacy|GetVisibilityInGame))
  (bind _enabledonminigamestart_2 (Variables|Legacy|GetEnabledOnMinigameStart))
  (Utilities|FlowControl|Sequence
    (:then_0
      ;[1 1 1] Split TimesCanTrigger_Override from TimesCanTrigger
      (if (!= _timescantrigger _timescantrigger_2)
        (LocalVariables|SetbArePropertiesUpdated true)
        (bind _output_get (Variables|UserOptions|SetTimesCanTrigger_Override (> _timescantrigger_2 0)))
        (if (not _output_get)
          (Variables|UserOptions|SetTimesCanTrigger _timescantrigger))))
    (:then_1
      ;[1 1 1] Split TriggerSoundIndex and TriggerSFX
      (if (!= _triggersoundindex _triggersoundindex_2)
        (Variables|Legacy|SetTrigger%20SFX (>= _triggersoundindex_2 0))
        (Variables|Legacy|SetTriggerSoundIndex _triggersoundindex)
        (LocalVariables|SetbArePropertiesUpdated true)))
    (:then_2
      ;[0 0.5658 1 0.25] Migrating Team and Class
      (if (!= (Variables|Legacy|GetClassIndex) _teamsetting)
        (Variables|UserOptions|Basic|SetActivatingClass (Creative|ClassOption|MakefromLegacyClassIndex (Variables|Legacy|GetClassIndex)))
        (Variables|Legacy|SetClassIndex _teamsetting)
        (LocalVariables|SetbArePropertiesUpdated true)))
    (:then_3
      (if (!= _classindex (Variables|Legacy|GetTeamSetting))
        (Variables|UserOptions|Basic|SetActivatingTeam (Creative|TeamOption|MakefromLegacyTeamIndex (Variables|Legacy|GetTeamSetting)))
        (Variables|Legacy|SetTeamSetting _classindex)
        (LocalVariables|SetbArePropertiesUpdated true)))
    (:then_4
      ;[1 1 1] Migrate Visibility
      (if (!= _visibilityingame _visibilityingame_2)
        (Variables|UserOptions|Basic|SetVisible%20in%20Game _visibilityingame_2)
        (Variables|Legacy|SetVisibilityInGame _visibilityingame)
        (LocalVariables|SetbArePropertiesUpdated true)))
    (:then_5
      ;[1 1 1] Migrate EnabledOnGameStart
      (if (!= _enabledonminigamestart _enabledonminigamestart_2)
        (Variables|UserOptions|SetEnabled%20On%20Game%20Start _enabledonminigamestart_2)
        (Variables|Legacy|SetEnabledOnMinigameStart _enabledonminigamestart)
        (LocalVariables|SetbArePropertiesUpdated true)))
    (:then_6
      ;[1 1 1] Migrate Reset Delay
      (if (!= _triggerresetdelay _triggerresetdelay_2)
        (Variables|UserOptions|SetReset%20Delay _triggerresetdelay_2)
        (Variables|Legacy|SetTriggerResetDelay _triggerresetdelay)
        (LocalVariables|SetbArePropertiesUpdated true)))
    (:then_7
      ;[1 1 1] Refresh Property Overrides if we updated any of our properties
      (if (LocalVariables|GetbArePropertiesUpdated)
        (OptionData|RefreshPropertyOverrides (Variables|Default|GetToyOptionsComponent))))))

(fn NotTriggeringSelf
  ;[1 1 1] Very basic check to make sure this isn't triggering itself.
  ; Players can easily get around this by having two triggers trigger each other, but this can be done with other devices as well, If a creator screws over their own experience, that's on them.
  ; Children will define their own implementations.
  (return true))

(fn OnRep_CurrentState
  (if (not (Components|Timeline|IsPlaying (Variables|Device_Trigger_Parent_V2|GetTriggered)))
    (Default|SetMaterialBasedonCurrentState)))

(fn OnRep_TriggerSFXOverride
  ; This has to be done so that we can notify the Local device that this has been updated and that it should use the TriggerSFXOverride instead. Otherwise, it will never use the TriggerSFXOverride we pass from Verse.
  (if (Minigame|IsinPlayMode self :!class /Script/FortniteGame.MinigameBlueprintLibrary)
    (Variables|Default|SetTriggerSFXOverriddenDuringPlay true)))

(fn OnRep_TriggerVFXOverride
  ; This has to be done so that we can notify the Local device that this has been updated and that it should use the TriggerVFXOverride instead. Otherwise, it will never use the TriggerVFXOverride we pass from Verse.
  (if (Minigame|IsinPlayMode self :!class /Script/FortniteGame.MinigameBlueprintLibrary)
    (Variables|Default|SetTriggerVFXOverriddenDuringPlay true)))

(fn PhysicalRetriggerEnterLogic
  (params (Triggering%20Object :type Actor%20Object%20Reference) (Triggering%20Component :type Primitive%20Component%20Object%20Reference))
  (if (Utilities|Array|ContainsItem (Variables|Default|GetPhysical%20Actor%20Collision%20Types) (Collision|GetCollisionObjectType Triggering%20Component))
    (if (Utilities|Array|ContainsItem (Variables|Default|GetTriggering%20Physical%20Actors) Triggering%20Object)
      (return false)
      (else
        (Utilities|Array|Add (Variables|Default|GetTriggering%20Physical%20Actors) Triggering%20Object)))
    (else
      _))
  (return true))

(fn PhysicalRetriggerLeaveLogic
  (params (Triggering%20Object :type Actor%20Object%20Reference))
  (Utilities|Array|RemoveItem (Variables|Default|GetTriggering%20Physical%20Actors) Triggering%20Object))

(fn ResetTriggerCount
  (bind _timescantrigger (Verse|GetNumTimesCanTrigger))
  (Variables|Default|SetRemainingTimesCanTrigger _timescantrigger)
  (Variables|Default|SetTimesTriggered 0))

(fn SetTriggerCollision
  (params (On :type Boolean))
  (return false))

(fn SetCurrentConstantMaterial
  (params (NewCurrentMaterial :type Material%20Interface%20Object%20Reference))
  (Rendering|Material|SetMaterial (Variables|Default|GetMesh) 1 NewCurrentMaterial)
  (CallFunction|ReleaseDynamicMaterial)
  (Variables|Materials|SetCurrentMaterial NewCurrentMaterial))

(fn SetInstigator
  (params (InInstigator :type Actor%20Object%20Reference))
  (Utilities|FlowControl|Sequence
    (:then_0
      (bind _ascontroller (Utilities|Casting|CastToController (Variables|GetInInstigator))
        (:then
          (LocalVariables|Setlocal%20Instigator _ascontroller))
        (:CastFailed
          (if (Utilities|DoesObjectImplementInterface (Variables|GetInInstigator) "/Script/FortniteGame.FortVehicleInterface")
            (bind _returnvalue (Vehicle|GetDriver(Message) (Variables|GetInInstigator)))
            (Utilities|IsValid _returnvalue
              (:Is%20Valid
                (LocalVariables|Setlocal%20Instigator (Pawn|GetController _returnvalue)))
              (:Is%20Not%20Valid
                (bind _returnvalue_2 (Vehicle|GetAllPassengers(Message) (Variables|GetInInstigator)))
                (bind _array_element (Utilities|Array|ForEachLoopwithBreak _returnvalue_2)
                  (:LoopBody
                    (Utilities|IsValid _array_element
                      (:Is%20Valid
                        (Utilities|IsValid (Pawn|GetController _array_element)
                          (:Is%20Valid
                            (LocalVariables|Setlocal%20Instigator (Pawn|GetController _array_element))))))))))
            (else
              (bind _aspawn (Utilities|Casting|CastToPawn (Variables|GetInInstigator)))
              (LocalVariables|Setlocal%20Instigator (Pawn|GetController _aspawn)))))))
    (:then_1
      ;[1 1 1] Sets separately rather than inline so if all of the instigator checks fail, we overwrite previous data to Null.
      (Variables|Default|SetTriggerInstigator (LocalVariables|Getlocal%20Instigator)))))

(fn SetMaterialBasedOnCurrentState
  (switch Utilities|FlowControl|Switch|SwitchonEEnabledState (Variables|Default|GetCurrentState)
    (:NewEnumerator0
      (CallFunction|SwitchToonMaterial))
    (:NewEnumerator1
      (CallFunction|SwitchToAnimatingMaterial))
    (:NewEnumerator2
      (CallFunction|SwitchToOffMaterial))))

(fn SetMaxTriggerCountOverride
  (params (Override :type Boolean))
  (Networking|SwitchHasAuthority
    (:Authority
      (Variables|UserOptions|SetTimesCanTrigger_Override Override))))

(fn SetMeshVisibilityAndCollision
  (params (bNewVisibility :type Boolean))
  (bind _mesh (Variables|Default|GetMesh))
  (if (Variables|GetbNewVisibility)
    (Collision|SetCollisionProfileName _mesh "FortDynamicMesh" true :!class /Script/Engine.PrimitiveComponent)
    (Collision|SetCollisionResponsetoChannel _mesh "ECC_GameTraceChannel11" "ECR_Ignore")
    (Collision|SetCollisionResponsetoChannel _mesh "ECC_GameTraceChannel17" "ECR_Block")
    (else
      (Collision|SetCollisionEnabled (Variables|Default|GetMesh) "NoCollision"))))

(fn SetNumTimesCanTrigger
  (params (TimesCanTrigger :type Integer))
  (Networking|SwitchHasAuthority
    (:Authority
      (bind _output_get (Variables|Verse|SetTimesCanTriggerOverride (Math|Integer|Clamp(Integer) TimesCanTrigger 0 10)))
      (Variables|Default|SetTimesCanTriggerOverridden (select (!= _output_get 0) "NewEnumerator2" "NewEnumerator1"))
      (CallFunction|ResetTriggerCount)
      (CallFunction|EndReset))))

(fn SetResetDelay
  (params (ResetDelay :type Float%20%28double-precision%29))
  (Networking|SwitchHasAuthority
    (:Authority
      (Variables|Verse|SetResetDelayOverride (Math|Float|Clamp(Float) ResetDelay 0.0 1.2e+03)))))

(fn SetShouldPlaySFX
  (params (InputPin :type Boolean))
  (Networking|SwitchHasAuthority
    (:Authority
      (Variables|Verse|SetTriggerSFXOverride InputPin))))

(fn SetShouldPlayVFX
  (params (Bool :type Boolean))
  (Networking|SwitchHasAuthority
    (:Authority
      (Variables|Verse|SetTriggerVFXOverride Bool))))

(fn SetTriggerDelay
  (params (Delay :type Float%20%28double-precision%29))
  (Networking|SwitchHasAuthority
    (:Authority
      (Variables|Verse|SetTriggerDelayOverride (Math|Float|Clamp(Float) Delay 0.0 1.2e+03)))))

(fn StartActivateAttempt
  (params (TriggerInstigator :type Controller%20Object%20Reference) (Do%20Trigger :type Delegate%20%28by%20ref%29))
  ;[0 0.432 1] Attempting to Activate the Device, checking against if it can trigger or not.
  (Networking|SwitchHasAuthority
    (:Authority
      (if (CreativeUserRestrictions|CanParticipantActorInteract (Variables|Default|GetCreative_UseRestriction_Component) TriggerInstigator true (CallFunction|CanTrigger) false)
        (Default|UnbindallEventsfromOnTrigger)
        (Default|BindEventtoOnTrigger Do%20Trigger)
        (CallFunction|SetInstigator TriggerInstigator)
        (bind _delay (Verse|GetTriggerDelay))
        (Variables|Default|SetbIsDelayed true)
        (if (> _delay 0.0)
          (if (not (Utilities|Time|IsValidTimerHandle (Variables|Default|GetTriggerDelayTimer)))
            (bind _returnvalue (Utilities|Time|SetTimerbyEvent (EventDispatchers|CreateEvent :!event SendActivationEvent) 0.01 false false _delay))
            (Variables|Default|SetTriggerDelayTimer _returnvalue)
            (Variables|Default|SetCurrentState "NewEnumerator1"))
          (else
            (CallFunction|SendActivationEvent)))))))

(fn TransmitEvery
  ;[0.8666 0.2533 1] Overridden by children that care about transmitting every X times.
  (return 1))

(fn TransmitMessage
  ;[1 1 1] Activate the Trigger that is attempting to fire.
  (Default|CallOnTrigger (Variables|Default|GetTriggerInstigator))
  (Default|UnbindallEventsfromOnTrigger)
  (return true))

(fn UpdateDynamicMaterial
  (Utilities|FlowControl|Sequence
    (:then_0
      (bind _channel_material_instance (Variables|Materials|GetChannel%20Material%20Instance)
        (:then
          ;[1 1 1] Material is getting disconnected from plane on object copy. This re-attaches it.
          (if (not (Utilities|Equal(Object) (Rendering|Material|GetMaterial (Variables|Default|GetChannelNumberPlane) 0) _channel_material_instance))
            (Rendering|Material|SetMaterial (Variables|Default|GetChannelNumberPlane) 0 (Variables|Materials|GetChannel%20Material%20Instance))))
        (:else
          (Utilities|BuildMaterial (Variables|Default|GetChannelNumberPlane) 0 0 (Variables|Materials|GetChannel%20Material%20Instance)))))
    (:then_1
      (bind _returnvalue (Class|FortMaterialCacheManager|GetCache (Variables|ChannelNumberUpdate|GetMaterial%20Cache%20Settings) (Variables|Materials|GetDefaultOnMaterial)))
      (Variables|Materials|SetMaterial%20Cache _returnvalue))))

(fn UpdateChannelNumber
  (params (Num :type Integer))
  (if (Variables|ChannelNumberUpdate|GetUsesChannelNumber)
    (ChannelNumberUpdate|UpdateDynamicMaterial)
    (Rendering|Material|SetScalarParameterValue (Variables|Materials|GetChannel%20Material%20Instance) "Score" (Math|Conversions|ToFloat(Integer) Num) :!class /Script/Engine.MaterialInstanceDynamic)
    (Rendering|Material|SetVectorParameterValue (Variables|Materials|GetChannel%20Material%20Instance) "Color" (Math|Color|SelectColor (Variables|ChannelNumberUpdate|GetChannelLEDColor) (Variables|ChannelNumberUpdate|GetChannelLEDColor_Unset) (> Num 0)) :!class /Script/Engine.MaterialInstanceDynamic)))

(fn UpdateVisibilityDuringPhase
  (Visibility|SetVisibilityComponents (Variables|Default|GetVisibleInGameComponent) (Utilities|Array|MakeArray (Variables|Default|GetMesh) (Variables|Default|GetChannelNumberPlane)) true)
  (Networking|SwitchHasAuthority
    (:Authority
      (Visibility|SetVisiblityDuringPhase (Variables|Default|GetVisibleInGameComponent) (select (Variables|UserOptions|Basic|GetVisible%20in%20Game) "NewEnumerator3" "NewEnumerator1")))))

(fn ConstructionScript
  (Variables|BuildingActor|SetbIsInvulnerable true)
  (|CallParentFunction))
