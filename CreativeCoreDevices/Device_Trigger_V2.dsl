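; Creative Trigger device that broadcasts OnTriggered when activated by overlap, damage, pulse, water, events or game phase.
; Authority-side overlaps pass retrigger and type filters to resolve an instigating controller, with possession callbacks covering players spawning on it.
; Requests funnel into StartActivateAttempt, where delayed-instigator mode selects a single broadcast or one per queued instigator.
; Option accessors, some noted for Verse, expose settings; damage triggers once per frame to absorb shotgun pellets.
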(event Custom|QueuedActivateTrigger
  (params (Instigator :type Controller%20Object%20Reference))
  (for _array_element (Variables|Default|GetQueuedInstigators)
    (GameplayEvent|BroadcastEvent (Variables|UserOptions|Events|GetOnTriggered) 0 (Utilities|Struct|MakeGameplayEventTrigger _array_element)))
  (Utilities|Array|Clear (Variables|Default|GetQueuedInstigators)))

(event Custom|RequestActivateTrigger
  (params (Instigator :type Controller%20Object%20Reference))
  (bind _delay (Verse|GetTriggerDelay))
  (CallFunction|StartActivateAttempt Instigator (Utilities|Select :Index (Variables|UserOptions|Advanced|GetDelayedTriggerInstigatorChoice) :NewEnumerator0 (delegate Custom|BroadcastActivateTrigger) :NewEnumerator1 (delegate Custom|BroadcastActivateTrigger) :NewEnumerator2 (select (> _delay 0) (EventDispatchers|CreateEvent :!event Custom|QueuedActivateTrigger) (delegate Custom|BroadcastActivateTrigger)))))

(event Custom|BroadcastActivateTrigger
  (params (Instigator :type Controller%20Object%20Reference))
  (GameplayEvent|BroadcastEvent (Variables|UserOptions|Events|GetOnTriggered) 0 (Utilities|Struct|MakeGameplayEventTrigger Instigator)))

(event EventMinigameStateChanged
  (params (New%20State :type EFortMinigameState%20Enum :default "PreGame"))
  ;[0 0.6006 1 0.5] When the game state changes, check to see if we need to activate, and if so, do.
  (|CallParentFunction New%20State)
  (if (Class|CreativeCommonDeviceFunctionLibrary2|ActivateonPhase (Variables|/Script/FortniteGame.FortMinigameLogicComponent|GetCurrentMinigame (Variables|Default|GetFortMinigameLogic)) (Variables|UserOptions|GetActivateOnGamePhase))
    (CallFunction|RequestActivateTrigger 0)))

(event PostMinigameStartedDelegate(FortMinigameLogic)
  ;[0 0.6006 1 0.5] When the game state changes, check to see if we need to activate, and if so, do.
  (CreativeDeviceHelper|ActivateonPhase|BranchActivateonGameStart (Variables|UserOptions|GetActivateOnGamePhase)
    (:True
      (CallFunction|RequestActivateTrigger 0))))

(event EventBeginPlay
  (|CallParentFunction)
  (Networking|SwitchHasAuthority
    (:Remote
      (Collision|SetCollisionEnabled (Variables|Default|GetRetriggerLimiterBox) "NoCollision") ; Disable collision on trigger for non-authority,
    )))

(event EventOnHitByMusicSequencer
  (params (InstigatedBy :type Controller%20Object%20Reference))
  (if (Variables|UserOptions|GetTriggeredByPulseTrigger)
    (CallFunction|RequestActivateTrigger InstigatedBy)))

(event EventOnDamageServer
  (params (Damage :type Float%20%28single-precision%29) (DamageTags :type Gameplay%20Tag%20Container%20Structure%20%28by%20ref%29) (Momentum :type Vector :default "0, 0, 0") (HitInfo :type Hit%20Result%20Structure%20%28by%20ref%29) (InstigatedBy :type Controller%20Object%20Reference) (DamageCauser :type Actor%20Object%20Reference) (EffectContext :type Gameplay%20Effect%20Context%20Handle%20Structure))
  (|CallParentFunction Damage DamageTags Momentum HitInfo InstigatedBy DamageCauser EffectContext)
  (if (Variables|UserOptions|GetTriggeredByDamage)
    ;[0.1719 0.6901 1] When triggered by damage, only accept one trigger per frame. This covers the case of Shotguns, where multiple pellets will be hitting at once.
    (Utilities|DoOncePerFrame
      (:%20%20
        (Variables|Actor|SetInstigator InstigatedBy)
        (CallFunction|RequestActivateTrigger InstigatedBy)))))

(event OnComponentBeginOverlap(StepBox)
  (params (OverlappedComponent :type Primitive%20Component%20Object%20Reference) (OtherActor :type Actor%20Object%20Reference) (OtherComp :type Primitive%20Component%20Object%20Reference) (OtherBodyIndex :type Integer) (bFromSweep :type Boolean) (SweepResult :type Hit%20Result%20Structure%20%28by%20ref%29))
  ;[1 1 1] Overlaps
  (Networking|SwitchHasAuthority
    (:Authority
      (bind _allowtrigger (CallFunction|PhysicalRetriggerEnterLogic OtherActor OtherComp)) ; Bouncing vehicle shouldn't trigger multiple times
      (if _allowtrigger
        (bind (_valid_attempt _controller) (CallFunction|CheckValidTriggerAttempt OtherActor))
        (if _valid_attempt
          (CallFunction|RequestActivateTrigger _controller))))))

(event OnComponentEndOverlap(RetriggerLimiterBox)
  (params (OverlappedComponent :type Primitive%20Component%20Object%20Reference) (OtherActor :type Actor%20Object%20Reference) (OtherComp :type Primitive%20Component%20Object%20Reference) (OtherBodyIndex :type Integer))
  ;[1 1 1] Overlaps
  (Networking|SwitchHasAuthority
    (:Authority
      (CallFunction|PhysicalRetriggerLeaveLogic OtherActor))))

(event PawnNowHasValidParticipantController(Creative_UseRestriction_Component)
  (params (FortPlayerPawn :type Fort%20Player%20Pawn%20Object%20Reference) (Controller :type Controller%20Object%20Reference))
  ;[1 1 1] Overlaps
  ;;[1 1 1] If a player spawns on a trigger, they won't have a controller yet. This callback will activate when they are possessed.
  (if (Collision|IsOverlappingActor (Variables|Default|GetStepBox) FortPlayerPawn :!class /Script/Engine.PrimitiveComponent)
    ;;;[1 0 0 0.5] The Trigger can be activated in the middle of the Possession. If's that's the case we wait until the Controller has Fully Possessed the Pawn
    (bind _asfort_player_controller (Utilities|Casting|CastToFortPlayerController Controller)
      (:then
        (if (not (Class|FortPlayerController|IsPawnAcknowlodged _asfort_player_controller))
          (Game|Player|BindEventtoOnPlayerPawnPossessed _asfort_player_controller (delegate Custom|ParticipantControllerAcknowledgesNewPawn))))
      (:CastFailed
        _))
    (CallFunction|RequestActivateTrigger Controller)))

(event Custom|ParticipantControllerAcknowledgesNewPawn
  (params (PossessedPawn :type Pawn%20Object%20Reference))
  (bind _returnvalue (Pawn|GetController PossessedPawn))
  ;[1 1 1] Overlaps
  ;;[1 1 1] If a player spawns on a trigger, they won't have a controller yet. This callback will activate when they are possessed.
  (Game|Player|UnbindEventfromOnPlayerPawnPossessed (Utilities|Casting|CastToFortPlayerController _returnvalue) (delegate Custom|ParticipantControllerAcknowledgesNewPawn))
  (if (Collision|IsOverlappingActor (Variables|Default|GetStepBox) PossessedPawn :!class /Script/Engine.PrimitiveComponent)
    (CallFunction|RequestActivateTrigger _returnvalue)))

(event Custom|TriggerFromEvent
  (params (Instigator :type Controller%20Object%20Reference))
  ;[0.6648 0 1] Receiver
  (CallFunction|RequestActivateTrigger Instigator))

(fn ActivateFromOtherDevice
  (params (Instigator :type Controller%20Object%20Reference) (Requester :type ERequesterType%20Enum :default "NewEnumerator0"))
  (if (Utilities|Select :Index Requester :NewEnumerator0 (Variables|UserOptions|GetTriggeredByPulseTrigger) :NewEnumerator1 (Variables|UserOptions|GetTriggeredByWater))
    (CallFunction|RequestActivateTrigger Instigator)
    (return true)
    (else
      (return false))))

(fn CanTrigger
  (bind _timescantrigger (Verse|GetNumTimesCanTrigger))
  (return (Utilities|Select :Index (Variables|UserOptions|Advanced|GetDelayedTriggerInstigatorChoice) :NewEnumerator0 (|CallParentFunction) :NewEnumerator1 (Math|Boolean|ANDBoolean (Class|CreativeCommonDeviceFunctionLibrary|IsinPlayMode self) (not (Variables|Default|GetbIsResetting)) (or (<= _timescantrigger 0) (> (Variables|Default|GetRemainingTimesCanTrigger) 0)) (Variables|Default|GetbIsEnabled)) :NewEnumerator2 (Math|Boolean|ANDBoolean (Class|CreativeCommonDeviceFunctionLibrary|IsinPlayMode self) (not (Variables|Default|GetbIsResetting)) (or (<= _timescantrigger 0) (> (Variables|Default|GetRemainingTimesCanTrigger) 0)) (Variables|Default|GetbIsEnabled)))))

(fn CheckCreature
  (params (Actor :type Actor%20Object%20Reference))
  (bind _asfort_aipawn (Utilities|Casting|CastToFortAIPawn Actor)
    (:then
      (return (Variables|UserOptions|GetTriggeredByCreatures) (Pawn|GetController _asfort_aipawn)))
    (:CastFailed
      (return false 0))))

(fn CheckForbiddenClasses
  (params (Actor :type Actor%20Object%20Reference))
  ;[1 1 1] Certain Classes should never activate the trigger. Check for these here.
  (for _array_element (Variables|Default|GetForbiddenClasses)
    (if (Utilities|IsValidSoftClassReference _array_element)
      (if (Utilities|ClassIsChildOf (Utilities|GetClass Actor) (Utilities|ResolveSoftReference _array_element))
        (return false))))
  (return true))

(fn CheckItem
  (params (Actor :type Actor%20Object%20Reference))
  (Utilities|FlowControl|Sequence
    (:then_0
      ;[0.3932 0.4388 1] Determine if the actor is an item and get it's controller
      (bind _asfort_projectile_base (Utilities|Casting|CastToFortProjectileBase Actor)
        (:then
          (bind _returnvalue (PlayerInfo|GetOwnerPlayerController _asfort_projectile_base :!class /Script/FortniteGame.FortProjectileBase))
          (LocalVariables|SetOwnerController _returnvalue))
        (:CastFailed
          (bind _asfort_pickup (Utilities|Casting|CastToFortPickup Actor)
            (:then
              (bind _returnvalue_2 (PlayerInfo|GetOwnerPlayerController _asfort_pickup :!class /Script/FortniteGame.FortPickup))
              (bind _output_get (LocalVariables|SetOwnerController _returnvalue_2))
              (Utilities|IsValid _output_get
                (:Is%20Not%20Valid
                  (Utilities|IsValid (Pickup|GetPawnWhoDroppedPickup _asfort_pickup)
                    (:Is%20Valid
                      (bind _asfort_player_controller (Utilities|Casting|CastToFortPlayerController (Actor|GetOwner (Pickup|GetPawnWhoDroppedPickup _asfort_pickup))))
                      (LocalVariables|SetOwnerController _asfort_player_controller))))))
            (:CastFailed
              (bind _asfort_physics_pawn_object (Utilities|Casting|CastToFortPhysicsPawnObject Actor)
                (:then
                  (Utilities|IsValid (Variables|/Script/Engine.Pawn|GetLastHitBy _asfort_physics_pawn_object)
                    (:Is%20Valid
                      (bind _asfort_player_controller_2 (Utilities|Casting|CastToFortPlayerController (Variables|/Script/Engine.Pawn|GetLastHitBy _asfort_physics_pawn_object)))
                      (LocalVariables|SetOwnerController _asfort_player_controller_2))))
                (:CastFailed
                  (bind _returnvalue_3 (GameplayTags|HasMatchingGameplayTag(Message) Actor "(TagName=\"Building.Type.Prop.PhysicsBeachBall\")"))
                  (if _returnvalue_3
                    (bind _asfort_player_controller_3 (Utilities|Casting|CastToFortPlayerController (Game|GetInstigatorController Actor)))
                    (LocalVariables|SetOwnerController _asfort_player_controller_3)
                    (else
                      ;[1 0 0.03856] The Actor is not an item, return
                      (return false 0))))))) ; If they're not a projectile, \ncheck if they're a pickup.
        )) ; Projectiles should count as items.
    )
    (:then_1
      ;[0.2747 1 0.1171] Return the found controller and whether it part of a valid team
      (return (and (StateChecks|IsValidTeam (Variables|Default|GetCreative_UseRestriction_Component) (LocalVariables|GetOwnerController)) (Variables|UserOptions|GetTriggeredByItems)) (LocalVariables|GetOwnerController)))))

(fn CheckPhysicsObject
  (params (Actor :type Actor%20Object%20Reference))
  (if (Variables|UserOptions|GetTriggeredByPhysicsProps)
    (if (Fortnite|IsPhysicallySimulatedIsland)
      (bind _returnvalue (FortPhysicsSimulationLibrary|IsPhysicsObjectActiveinSimulation (Variables|GetActor)))
      (if _returnvalue
        (return true)
        (else
          (return false)))
      (else
        _)))
  (return false))

(fn CheckPlayer
  (params (Actor :type Actor%20Object%20Reference))
  (bind _aspawn (Utilities|Casting|CastToPawn Actor)
    (:then
      (Utilities|Casting|CastToFortAIPawn (Variables|GetActor)
        (:then
          _)
        (:CastFailed
          (if (not (Utilities|DoesObjectImplementInterface Actor "/Script/FortniteGame.FortVehicleInterface"))
            (Utilities|IsinVehicle _aspawn
              (:true
                (return false))
              (:false
                (return (Variables|UserOptions|GetTriggeredByPlayer))))
            (else
              _))))
      (return false))
    (:CastFailed
      _)))

(fn CheckValidTriggerAttempt
  (params (Actor :type Actor%20Object%20Reference))
  (bind _actor (Variables|GetActor))
  (bind (_pass _controller) (CheckValidTriggerAttempt|CheckCreature _actor))
  (bind _actor_2 (Variables|GetActor))
  (bind (_isitem _fortplayercontroller) (CheckValidTriggerAttempt|CheckItem _actor_2))
  (if (CheckValidTriggerAttempt|CheckForbiddenClasses Actor)
    (if _isitem
      (return true _fortplayercontroller)
      (elif (or (CheckValidTriggerAttempt|CheckPlayer (Variables|GetActor)) (CheckValidTriggerAttempt|CheckVehicle (Variables|GetActor)))
        (bind (_valid_controller _fortplayerpawn _controller_2) (CreativeUserRestrictions|ActorHasValidParticipantController (Variables|Default|GetCreative_UseRestriction_Component) (Variables|GetActor) false true true))
        (return (or (CheckValidTriggerAttempt|CheckVehicle (Variables|GetActor)) _valid_controller) _controller_2)
        (elif _pass
          (return true _controller)
          (elif (or (CheckValidTriggerAttempt|CheckPhysicsObject (Variables|GetActor)) (CallFunction|CheckCarryableObject (Variables|GetActor)))
            (return true 0)))))
    (else
      _))
  (return false 0))

(fn CheckVehicle
  (params (Actor :type Actor%20Object%20Reference))
  (if (Utilities|DoesObjectImplementInterface Actor "/Script/FortniteGame.FortVehicleInterface")
    (return (Variables|UserOptions|GetTriggeredByVehicles))
    (else
      (return false))))

(fn Check_Carryable_Object
  (params (Actor_IN :type Actor%20Object%20Reference))
  (bind _actor_in (Variables|GetActor_IN))
  (if (Variables|UserOptions|GetTriggeredByCarryableObjects)
    (if (Fortnite|IsPhysicallySimulatedIsland)
      (Utilities|IsValid (Actor|GetComponentbyClass _actor_in "/Script/FortniteGame.FortPhysicsCarryableComponent")
        (:Is%20Valid
          (return true))
        (:Is%20Not%20Valid
          _))
      (else
        (Utilities|IsA(soft) _actor_in "/Game/Athena/Items/HeldObjects/HeldObjectVersion2/BGA_HeldObject_Physics_Parent.BGA_HeldObject_Physics_Parent_C"
          (:True
            _)
          (:False
            _))))
    (return false)
    (else
      (return false))))

(fn GetCanReceiveDamageWhileInvisible
  (return (Variables|UserOptions|GetbReceiveDamageWhenInvisible)))

(fn GetCollisionComponent
  (return (Variables|Default|GetMesh)))

(fn GetDelayType
  ; When called from Verse, we transcribe DelayedTriggerInstigatorChoice to an Int, so that we can assign the proper enum Value in Verse for the Return Value.
  (switch Utilities|FlowControl|Switch|SwitchonEDelayedTriggerInstigatorChoice (Variables|UserOptions|Advanced|GetDelayedTriggerInstigatorChoice)
    (:NewEnumerator0
      (LocalVariables|SetDelayTypeInt 0))
    (:NewEnumerator1
      (LocalVariables|SetDelayTypeInt 1))
    (:NewEnumerator2
      (LocalVariables|SetDelayTypeInt 2)))
  (return (LocalVariables|GetDelayTypeInt)))

(fn GetInteractionCountPerTrigger
  (return (Variables|UserOptions|GetTransmit%20Every%20X%20Triggers)))

(fn GetShouldAllowCreatures
  (return (Variables|UserOptions|GetTriggeredByCreatures)))

(fn GetShouldAllowDamage
  (return (Variables|UserOptions|GetTriggeredByDamage)))

(fn GetShouldAllowItems
  (return (Variables|UserOptions|GetTriggeredByItems)))

(fn GetShouldAllowPlayers
  (return (Variables|UserOptions|GetTriggeredByPlayer)))

(fn GetShouldAllowPulseTrigger
  (return (Variables|UserOptions|GetTriggeredByPulseTrigger)))

(fn GetShouldAllowVehicles
  (return (Variables|UserOptions|GetTriggeredByVehicles)))

(fn GetShouldAllowWater
  (return (Variables|UserOptions|GetTriggeredByWater)))

(fn IsTargetableByLinearMover
  (return true))

(fn NotTriggeringSelf
  (if (and (Class|FortGameplayMessageComponentBase|IsChannelIdSet (Variables|Default|GetTriggerWhenReceived)) (Class|FortGameplayMessageComponentBase|IsChannelIdSet (Variables|Default|GetChannelTrigger)))
    (return (!= (Class|FortGameplayMessageComponentBase|GetMessageChannelIndex (Variables|Default|GetTriggerWhenReceived)) (Class|FortGameplayMessageComponentBase|GetMessageChannelIndex (Variables|Default|GetChannelTrigger))))
    (else
      (return true))))

(fn SetTriggerCollision
  (params (On :type Boolean))
  (return false))

(fn SetCanReceiveDamageWhileInvisible
  (params (CanReceiveDamageWhileInvisible :type Boolean))
  (Variables|UserOptions|SetbReceiveDamageWhenInvisible CanReceiveDamageWhileInvisible))

(fn SetDelayType
  (params (DelayType%20Int :type Integer))
  ; When this function is called in Verse, we transcribe the enums from the Verse Class to an Int. We use the int to determine what the enum value should be in the Blueprint and set DelayedTriggerInstigatorChoice.
  (switch int DelayType%20Int
    (:Default
      _)
    (:0
      _)
    (:1
      (Variables|UserOptions|Advanced|SetDelayedTriggerInstigatorChoice "NewEnumerator1")
      (return))
    (:2
      (Variables|UserOptions|Advanced|SetDelayedTriggerInstigatorChoice "NewEnumerator2")))
  (Variables|UserOptions|Advanced|SetDelayedTriggerInstigatorChoice "NewEnumerator0"))

(fn SetInstigator
  (params (InInstigator :type Actor%20Object%20Reference))
  (bind _triggerinstigator (Variables|Default|GetTriggerInstigator))
  (if (> (Variables|UserOptions|GetTriggerDelay) 0)
    (switch Utilities|FlowControl|Switch|SwitchonEDelayedTriggerInstigatorChoice (Variables|UserOptions|Advanced|GetDelayedTriggerInstigatorChoice)
      (:NewEnumerator0
        (Utilities|IsValid (Variables|Default|GetTriggerInstigator)
          (:Is%20Not%20Valid
            (|CallParentFunction (Variables|GetInInstigator)))))
      (:NewEnumerator1
        (|CallParentFunction (Variables|GetInInstigator)))
      (:NewEnumerator2
        (|CallParentFunction (Variables|GetInInstigator))
        (Utilities|IsValid _triggerinstigator
          (:Is%20Valid
            (Utilities|Array|AddUnique (Variables|Default|GetQueuedInstigators) _triggerinstigator)))))
    (else
      (|CallParentFunction (Variables|GetInInstigator)))))

(fn SetInteractionCountPerTrigger
  (params (Interaction%20Count%20Per%20Trigger :type Integer))
  (Variables|UserOptions|SetTransmit%20Every%20X%20Triggers Interaction%20Count%20Per%20Trigger))

(fn SetMeshVisibilityAndCollision
  (params (bNewVisibility :type Boolean))
  (bind _mesh (Variables|Default|GetMesh))
  (bind _mesh_2 (Variables|Default|GetMesh))
  (Rendering|SetVisibility (Variables|Default|GetChannelNumberPlane) (Variables|GetbNewVisibility) false)
  (if (Variables|GetbNewVisibility)
    (Collision|SetCollisionProfileName _mesh "FortDynamicMesh" true :!class /Script/Engine.PrimitiveComponent)
    (Collision|SetCollisionResponsetoChannel _mesh "ECC_GameTraceChannel11" "ECR_Ignore")
    (Collision|SetCollisionResponsetoChannel _mesh "ECC_GameTraceChannel17" "ECR_Block")
    (elif (Variables|UserOptions|GetbReceiveDamageWhenInvisible)
      (Collision|SetCollisionProfileName _mesh_2 "FortTriggerOnlyWeapons" true :!class /Script/Engine.PrimitiveComponent)
      (Collision|SetCollisionResponsetoChannel _mesh_2 "ECC_GameTraceChannel8" "ECR_Block")
      (if (Variables|UserOptions|GetTriggeredByWater)
        (Collision|SetCollisionResponsetoChannel _mesh_2 "ECC_WorldStatic" "ECR_Overlap") ; Water volume uses World Static channel to overlap.
      )
      (elif (Variables|UserOptions|GetTriggeredByWater)
        (Collision|SetCollisionEnabled _mesh_2 "QueryOnly")
        (Collision|SetCollisionResponsetoAllChannels _mesh_2 "ECR_Ignore")
        (else
          (Collision|SetCollisionEnabled (Variables|Default|GetMesh) "NoCollision"))))))

(fn SetShouldAllowCreatures
  (params (ShouldAllowCreatures :type Boolean))
  (Variables|UserOptions|SetTriggeredByCreatures ShouldAllowCreatures))

(fn SetShouldAllowDamage
  (params (ShouldAllowDamage :type Boolean))
  (Variables|UserOptions|SetTriggeredByDamage ShouldAllowDamage))

(fn SetShouldAllowItems
  (params (ShouldAllowItems :type Boolean))
  (Variables|UserOptions|SetTriggeredByItems ShouldAllowItems))

(fn SetShouldAllowPlayers
  (params (ShouldAllowPlayer :type Boolean))
  (Variables|UserOptions|SetTriggeredByPlayer ShouldAllowPlayer))

(fn SetShouldAllowPulseTrigger
  (params (ShouldAllowPulseTrigger :type Boolean))
  (Variables|UserOptions|SetTriggeredByPulseTrigger ShouldAllowPulseTrigger))

(fn SetShouldAllowVehicles
  (params (ShouldAllowVehicles :type Boolean))
  (Variables|UserOptions|SetTriggeredByVehicles ShouldAllowVehicles))

(fn SetShouldAllowWater
  (params (ShouldAllowWater :type Boolean))
  (Variables|UserOptions|SetTriggeredByWater ShouldAllowWater))

(fn ShouldInvisiblePropsBeAllowed
  (return false))

(fn TransmitEvery
  (return (Variables|UserOptions|GetTransmit%20Every%20X%20Triggers)))

(fn ConstructionScript
  (|CallParentFunction)
  (if (Editor|InEditor)
    (CallFunction|InitializeDevice)
    (ChannelNumberUpdate|UpdateChannelNumber (Math|Integer|Max(Integer) (Class|FortGameplayMessageComponentBase|GetMessageChannelIndex (Variables|Default|GetChannelTrigger))))))
