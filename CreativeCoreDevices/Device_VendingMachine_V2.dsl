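; Creative vending machine device that sells configured items to players for resource costs during a minigame.
; Outside play, pickups dropped on its DropBox are queued into the spawner component's list, with accept/reject screen and suck-in effects.
; In play, interaction or triggers spawn the selected item, with cost checks on interaction, while melee hits, triggers or auto-cycle change selection.
; Clients drive a dynamic screen material showing icon, rarity color and price.
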(event Custom|Event%20On%20Placed
  (Audio|PlaySoundatLocation "/Game/Sounds/Creative/Gadgets/Traps/Trap_ItemSpawner_Placed_Cue.Trap_ItemSpawner_Placed_Cue" (Transformation|GetActorLocation)))

(event EventBeginPlay
  (bind _minigame_spawner_component (Variables|Default|GetMinigame_Spawner_Component))
  (|CallParentFunction)
  (Networking|SwitchHasAuthority
    (:Authority
      (Default|BindEventtoAddedtoSpawnListDispatcher _minigame_spawner_component (EventDispatchers|CreateEvent :!event Custom|OnAddedToSpawnList))
      (Default|BindEventtoOnSpawnDispatcher _minigame_spawner_component (EventDispatchers|CreateEvent :!event Custom|OnSpawn))
      (Default|BindEventtoOnGrantDispatcher _minigame_spawner_component (EventDispatchers|CreateEvent :!event Custom|OnGranted))
      ;[1 1 1] Bind toy options events
      (OptionData|BindEventtoOnAnyPropertyChanged (Variables|Default|GetToyOptionsComponent) (EventDispatchers|CreateEvent :!event ResaveActor))
      (OptionData|AssignOnPropertiesLoaded (Variables|Default|GetToyOptionsComponent) (EventDispatchers|CreateEvent :!event ResaveActor))
      (CallFunction|StartAutoItemCycle))
    (:Remote
      ;[1 1 1] Bind minigame events
      (Minigame|BindEventtoMinigameStateChangedDelegate (Variables|Default|GetFortMinigameProgress) (delegate Custom|MinigameStateChanged))
      (CallFunction|InitializeFX)
      (CallFunction|UpdateVendingMachineScreenFX))))

(event Custom|MinigameStateChanged
  (params (Minigame :type Fort%20Minigame%20Object%20Reference) (NewMinigameState :type EFortMinigameState%20Enum :default "PreGame"))
  ;[1 1 1] Bind minigame events
  (CallFunction|UpdateMinigameVisuals NewMinigameState)
  (CallFunction|UpdateVendingMachineScreenFX))

(event MinigameStartedDelegate(FortMinigameProgress)
  (bind _minigame_spawner_component (Variables|Default|GetMinigame_Spawner_Component))
  ;[0 1 0] When minigame starts
  (Networking|SwitchHasAuthority
    (:Authority
      (Variables|Default|SetItemSpawnLocation (+ (Math|Vector|RotateVector (Variables|Default|GetItemSpawnLocationOffset) (Transformation|GetActorRotation)) (Transformation|GetWorldLocation (Variables|Default|GetDropBox))))
      (CallFunction|InitializeDefaultSpawnList)
      ;[0 1 0] Initial Spawn Timer setup
      (Class|MinigameSpawnerComponent|PauseSpawner _minigame_spawner_component)
      (Variables|/Game/Athena/Items/Traps/MinigameSpawner/Minigame_Spawner_Component.Minigame_Spawner_Component_C|SetbIgnoreSpawnTimer true _minigame_spawner_component)
      (Variables|Default|SetbDeviceEnabled (Variables|UserOptions|GetEnabled%20At%20Game%20Start))
      (Variables|Default|SetSelectedItemToSpawnIndex 0)
      (Variables|Default|SetbCapacityDisplayVisible false)
      ;[0 1 0] Override weapon ammunition
      (Default|TryOverrideWeaponAmmoonSpawnParams (Variables|Default|GetMinigame_Spawner_Component) (Variables|UserOptions|Basic|GetInitialWeaponAmmo_Override) (Variables|UserOptions|Basic|GetInitialWeaponAmmo)))))

(event OnAddedtoMinigame(FortMinigameProgress)
  (params (Minigame :type Fort%20Minigame%20Object%20Reference))
  ;[1 0.1082 0 0.25] Disable Collision on drop area during Edit mode, as overlaps are expensive.
  (Minigame|BindEventtoOnPlayModeChangedDelegate Minigame (delegate Custom|OnPlayModeChanged))
  (CallFunction|OnPlayModeChanged Minigame (Minigame|IsinPlayMode Minigame :!class /Script/FortniteGame.FortMinigame)))

(event Custom|OnPlayModeChanged
  (params (Minigame :type Fort%20Minigame%20Object%20Reference) (bIsInPlayMode :type Boolean))
  ;[1 0.1082 0 0.25] Disable Collision on drop area during Edit mode, as overlaps are expensive.
  (Collision|SetCollisionEnabled (Variables|Default|GetDropBox) (select bIsInPlayMode "NoCollision" "QueryOnly")))

(event MinigameEndedDelegate(FortMinigameProgress)
  ;[1 0.4 0] When minigame ends
  (Networking|SwitchHasAuthority
    (:Authority
      (Class|MinigameSpawnerComponent|ResetSpawner (Variables|Default|GetMinigame_Spawner_Component) true)
      (Variables|Default|SetSelectedItemToSpawnIndex (- (Utilities|Array|Length (Variables|/Script/FortniteGame.FortMinigameSpawnerComponent|GetToSpawnList (Variables|Default|GetMinigame_Spawner_Component))) 1))
      (Variables|Default|SetbDeviceEnabled true)
      (Variables|Default|SetbCapacityDisplayVisible true))))

(event OnComponentHit(DropBox)
  (params (HitComponent :type Primitive%20Component%20Object%20Reference) (OtherActor :type Actor%20Object%20Reference) (OtherComp :type Primitive%20Component%20Object%20Reference) (NormalImpulse :type Vector :default "0, 0, 0") (Hit :type Hit%20Result%20Structure%20%28by%20ref%29))
  (bind _itemthathit (Variables|Default|GetItemThatHit))
  (bind _itemstobeaddedqueue (Variables|Default|GetItemsToBeAddedQueue))
  ;[1 1 1] Handles item being added from a hit event.  Hit captures all pick-ups
  (Networking|SwitchHasAuthority
    (:Authority
      (bind _binprogress (CallFunction|IsMinigameinProgress))
      (if (not _binprogress)
        (bind _output_get (Variables|Default|SetItemThatHit OtherActor))
        (bind _asfort_pickup_athena (Utilities|Casting|CastToFortPickupAthena _output_get))
        (Utilities|IsPickupValid _asfort_pickup_athena (Variables|Default|GetInvalidPickupGameplayTagContainer)
          (:IsValid
            (if (and (not (Pickup|PickedUp _asfort_pickup_athena)) (> (Variables|Default|GetItemLimit) (Utilities|Array|Length (Variables|/Script/FortniteGame.FortMinigameSpawnerComponent|GetToSpawnList (Variables|Default|GetMinigame_Spawner_Component)))))
              (if (Interaction|BlueprintCanInteract (Pickup|GetPawnWhoDroppedPickup _asfort_pickup_athena))
                (Pickup|SetPickedUp _asfort_pickup_athena true)
                ;;[1 1 1] Every bounce of the pickup causes a hit trigger.  This ensures rest of script only fires once
                (if (not (Utilities|Equal(Object) (Variables|Default|GetItemThatHit) (Variables|Default|GetItemToBeAdded)))
                  (Variables|Default|SetItemToBeAdded _itemthathit)
                  (Utilities|Array|Add (Variables|Default|GetItemsToBeAddedQueue) _itemthathit)
                  (Variables|Default|SetVendingMachineState "AcceptingItem")
                  (Utilities|FlowControl|Delay 1.0) ; Ensure object still exists so FX can start before object is destroyed
                  (Variables|Default|SetVendingMachineState "DisplayingItem")
                  ;;[1 1 1] Loop through Queue array to be sure to add every item if items are added in rapid succession
                  (if (> (Utilities|Array|Length _itemstobeaddedqueue) 0)
                    (CallFunction|AddtoSpawnList (Utilities|Array|Get(acopy) _itemstobeaddedqueue)))
                  (Variables|Default|SetVendingMachineState "AcceptedItem")
                  (CallFunction|ResetAutoItemCycle)
                  (Utilities|FlowControl|Delay 2.0)
                  (CallFunction|ResetAutoItemCycle)
                  (Variables|Default|SetVendingMachineState "DisplayingItem")
                  (Variables|Default|SetSelectedItemToSpawnIndex (- (Utilities|Array|Length (Variables|/Script/FortniteGame.FortMinigameSpawnerComponent|GetToSpawnList (Variables|Default|GetMinigame_Spawner_Component))) 1)))
                (else
                  _))
              (else
                _))
            (CallFunction|ResetAutoItemCycle)
            (Variables|Default|SetVendingMachineState "RejectedItem")
            (Utilities|FlowControl|Delay 1.5)
            (CallFunction|ResetAutoItemCycle)
            (Variables|Default|SetVendingMachineState "DisplayingItem")))))))

(event Custom|Enable%20When%20Receiving%20From
  (params (Instigator :type Controller%20Object%20Reference))
  (Networking|SwitchHasAuthority
    (:Authority
      (bind _binprogress (CallFunction|IsMinigameinProgress))
      (if _binprogress
        (if (not (Variables|Default|GetbDeviceEnabled))
          (Variables|Default|SetbDeviceEnabled true))))))

(event Custom|Disable%20When%20Receiving%20From
  (params (Instigator :type Controller%20Object%20Reference))
  (Networking|SwitchHasAuthority
    (:Authority
      (bind _binprogress (CallFunction|IsMinigameinProgress))
      (if _binprogress
        (if (Variables|Default|GetbDeviceEnabled)
          (Variables|Default|SetbDeviceEnabled false))))))

(event Custom|Spawn%20Item%20When%20Receiving%20From
  (params (Instigator :type Controller%20Object%20Reference))
  (Networking|SwitchHasAuthority
    (:Authority
      (if (Variables|Default|GetbDeviceEnabled)
        (bind _binprogress (CallFunction|IsMinigameinProgress))
        (if _binprogress
          (bind _bitemspawned (Class|MinigameSpawnerComponent|SpawnManuallySelectedIndex (Variables|Default|GetMinigame_Spawner_Component) (Variables|Default|GetSelectedItemToSpawnIndex) 0))
          (if _bitemspawned
            (GameplayEvent|BroadcastEvent (Variables|UserOptions|Events|GetOnItemSpawned) 0 (Utilities|Struct|MakeGameplayEventItemSpawned Instigator))))))))

(event Custom|Cycle%20to%20Next%20Item%20When%20Receiving%20From
  (params (Instigator :type Controller%20Object%20Reference))
  (Networking|SwitchHasAuthority
    (:Authority
      (if (Variables|Default|GetbDeviceEnabled)
        (bind _binprogress (CallFunction|IsMinigameinProgress))
        (if _binprogress
          (CallFunction|SwitchToNextItem))))))

(event OnAnyPropertyChanged(ToyOptionsComponent)
  (CallFunction|LoadModel)
  (CallFunction|UpdateUseRestrictionComponent))

(event OnPropertiesLoaded(ToyOptionsComponent)
  (Variables|Default|SetSelectedItemToSpawnIndex (- (Utilities|Array|Length (Variables|/Script/FortniteGame.FortMinigameSpawnerComponent|GetToSpawnList (Variables|Default|GetMinigame_Spawner_Component))) 1))
  (CallFunction|LoadModel)
  (CallFunction|UpdateUseRestrictionComponent))

(event EditorPreview|EventSetupForEditorPreview
  (Variables|Default|SetSelectedItemToSpawnIndex (- (Utilities|Array|Length (Variables|/Script/FortniteGame.FortMinigameSpawnerComponent|GetToSpawnList (Variables|Default|GetMinigame_Spawner_Component))) 1))
  (CallFunction|LoadModel)
  (CallFunction|UpdateUseRestrictionComponent))

(event Interaction|EventBlueprintServerNotifyStartLongUse
  (params (InteractingPawn :type Fort%20Player%20Pawn%20Object%20Reference))
  ;[0.1161 1 0.3371] Cache the interacting pawn and their selection when they begin to interact with this actor and clean up the cache when the interaction ends (either because it was interrupted or completed).
  (if (Variables|Default|GetbSelectItemIndexOnInteractBegin)
    (Utilities|Map|Add (Variables|Default|GetCachedSelectedItemIndexPerInteractingPawn) InteractingPawn (Variables|Default|GetSelectedItemToSpawnIndex))))

(event Interaction|EventBlueprintServerNotifyEndLongUse
  (params (InteractingPawn :type Fort%20Player%20Pawn%20Object%20Reference) (bUseCompleted :type Boolean))
  ;[0.1161 1 0.3371] Cache the interacting pawn and their selection when they begin to interact with this actor and clean up the cache when the interaction ends (either because it was interrupted or completed).
  (if (Variables|Default|GetbSelectItemIndexOnInteractBegin)
    (Utilities|Map|Remove (Variables|Default|GetCachedSelectedItemIndexPerInteractingPawn) InteractingPawn)))

(event Interaction|EventBlueprintOnInteract
  (params (InteractingPawn :type Fort%20Pawn%20Object%20Reference) (InteractionBeingAttempted :type EInteractionBeingAttempted%20Enum :default "FirstInteraction"))
  ;[0.1816 0.351 1] On Interact spawn a weapon
  (bind _binprogress (CallFunction|IsMinigameinProgress))
  (if _binprogress
    (bind _selectedcostcomponent (CallFunction|GetSelectedCostComponent InteractingPawn))
    (if (Cost|PassesCostCheck _selectedcostcomponent InteractingPawn)
      (bind _bitemspawned (Class|MinigameSpawnerComponent|SpawnManuallySelectedIndex (Variables|Default|GetMinigame_Spawner_Component) (Utilities|GetSelectedItemIndexforPawn InteractingPawn) (CallFunction|GetInventoryOwnerfromPawn InteractingPawn)))
      (if _bitemspawned
        (GameplayEvent|BroadcastEvent (Variables|UserOptions|Events|GetOnItemSpawned) 0 (Utilities|Struct|MakeGameplayEventItemSpawned (Pawn|GetController InteractingPawn)))
        (bind _resourcestaken (CallFunction|TakeResourcesfromPlayer InteractingPawn))
        (if _resourcestaken
          (Variables|Default|SetLastInteractingPawn InteractingPawn))))))

(event VendingMachine|EventOnCycleToNextItem
  ;[1 1 1] Handle melee weapons hitting vending machine and cycle to next item to spawn
  (CallFunction|SwitchToNextItem)
  (CallFunction|ResetAutoItemCycle))

(event Custom|OnAddedToSpawnList
  (params (Added%20Actor :type Actor%20Object%20Reference))
  (CallFunction|AddedtoSpawnListFX Added%20Actor)
  (CallFunction|ResaveActor))

(event Custom|OnSpawn
  (params (Spawned%20Actor :type Actor%20Object%20Reference))
  (Networking|SwitchHasAuthority
    (:Authority
      (CallFunction|QuestProgression Spawned%20Actor)
      (Transformation|SetActorRotation Spawned%20Actor (Math|Rotator|MakeRotator 0.0 0.0 (.yaw (Math|Rotator|CombineRotators (Transformation|GetActorRotation) (Math|Rotator|MakeRotator 0.0 0.0 9e+01)))) false)
      (Utilities|IsValid (Variables|Default|GetLastInteractingPawn)
        (:Is%20Valid
          (bind _spawnedpickup (Class|MinigameSpawnerComponent|GetLatestSpawnedPickup (Variables|Default|GetMinigame_Spawner_Component)))
          (bind _asfort_player_pawn (Utilities|Casting|CastToFortPlayerPawn (Variables|Default|GetLastInteractingPawn))
            (:then
              (Class|FortPlayerPawn|ServerHandlePickup _asfort_player_pawn _spawnedpickup 0.0 "0, 0, 0" false)
              (Variables|Default|SetLastInteractingPawn 0)
              (CallFunction|OnSpawnFX))
            (:CastFailed
              (CallFunction|OnSpawnFX))))
        (:Is%20Not%20Valid
          (CallFunction|OnSpawnFX))))))

(event Custom|OnGranted
  (Networking|SwitchHasAuthority
    (:Authority
      (CallFunction|QuestProgression 0)
      (Variables|Default|GetLastInteractingPawn
        (:then
          (Variables|Default|SetLastInteractingPawn 0))
        (:else
          _))
      (CallFunction|OnSpawnFX))))

(event VendingMachine|EventOnPlayerInventoryChange
  (bind _vendingscreenmid (Variables|Default|GetVendingScreenMID)
    (:then
      (CallFunction|UpdatePriceTextColorBasedonAffordability _vendingscreenmid))
    (:else)))

(event Custom|LoadModel
  (bind _model (Variables|UserOptions|Basic|GetModel))
  (bind _modelarray (Variables|Default|GetModelArray))
  (bind (_value _returnvalue) (Utilities|Map|Find _modelarray _model))
  (if _returnvalue
    (bind _loaded_object (Creative|Loading|AsyncLoadEditorSafe _value)
      (:Loaded
        (bind _asstatic_mesh (Utilities|Casting|CastToStaticMesh _loaded_object))
        (Components|StaticMesh|SetStaticMesh (Variables|Debug|GetStaticMeshComponent) _asstatic_mesh)
        (for _index (range (+ (Rendering|Material|GetNumMaterials (Variables|Debug|GetStaticMeshComponent)) 1))
          (Rendering|Material|SetMaterial (Variables|Debug|GetStaticMeshComponent) _index (StaticMesh|GetMaterial _asstatic_mesh _index)))
        (CallFunction|InitializeFX)
        (CallFunction|UpdateVendingMachineScreenFX)))))

(event Custom|OnAcceptingItemsFX
  (bind _itemtobeadded (Variables|Default|GetItemToBeAdded))
  ;[1 1 1] FX for when an item disappears after being dropped on the spawner.
  (Networking|CanExecuteCosmeticEvents
    (:True
      (if (Utilities|IsValid _itemtobeadded :!class /Script/Engine.KismetSystemLibrary)
        (Audio|PlaySoundatLocation (Variables|Sound|GetObjectAddedToSpawnerSound) (Transformation|GetActorLocation _itemtobeadded)) ; ObjectAddedToSpawnerSound
        (bind _asfort_pickup (Utilities|Casting|CastToFortPickup _itemtobeadded))
        (bind _asb_pickups_parent (Utilities|Casting|CastToB_Pickups_Parent (Utilities|Array|Get(acopy) (Actor|GetAttachedActors _asfort_pickup true false))))
        (Variables|Default|SetCurrentPickup _asb_pickups_parent)
        (Variables|Default|SetPickedUpWorldLocation (Transformation|GetActorLocation _asb_pickups_parent))
        (Class|BPickupsParent|SetActiveBackgroundParticleSystem _asb_pickups_parent false false)
        (bind _asb_pickups_default (Utilities|Casting|CastToB_Pickups_Default _asb_pickups_parent))
        (Class|BPickupsParent|HideBackgroundandSpotlight _asb_pickups_default)
        (CallFunction|PlaySetTrapActiveVisuals)))))

(event Custom|AddedToSpawnListFX
  (params (Added%20Actor :type Actor%20Object%20Reference))
  ;[1 1 1] Fires when item is actually added to item spawn list (item that was dropped no longer exists at this point)  Good for an effect on the pad to make it look like the item was received by the receptacle.
  (Networking|CanExecuteCosmeticEvents
    (:True
      (Audio|PlaySoundatLocation (Variables|Sound|GetObjectAddedToSpawnListSound) (Transformation|GetWorldLocation (Variables|Default|GetDropBox))) ; ObjectAddedToSpawnListSound
    )))

(event Custom|PlaySetTrapActiveVisuals
  (bind _staticmeshcomponent (Variables|Debug|GetStaticMeshComponent))
  (bind (_min _max) (Components|StaticMesh|GetLocalBounds _staticmeshcomponent))
  (bind _returnvalue (- _max _min))
  (bind _returnvalue_2 (/ _returnvalue 2.0))
  (bind _returnvalue_3 (Transformation|GetWorldLocation _staticmeshcomponent))
  (bind (_returnvalue_x _returnvalue_y _returnvalue_z) (Math|Vector|BreakVector (+ (Variables|Default|GetBackpackLocationOffset) (+ _returnvalue_2 _returnvalue_3))))
  (bind _returnvalue_4 (+ _returnvalue_2 _returnvalue_3))
  (bind _pulsecolor (Variables|Default|GetPulseColor))
  (bind _currentraritycolor (Variables|Default|GetCurrentRarityColor))
  ;[1 1 1] Suck items in effects, black holes items towards vending machine, flashes screen
  (bind (_direction _trapvisuals _movemeshtocenter _sucktobackpack) (Timeline|Timeline_0 PlayFromStart)
    (:Update
      (Rendering|Material|SetVectorParameterValue (Variables|Default|GetVendingScreenMID) "ScreenColor" (Math|Color|Lerp(LinearColor) _pulsecolor _currentraritycolor _trapvisuals) :!class /Script/Engine.MaterialInstanceDynamic)
      (Rendering|Components|Light|SetLightColor (Variables|Default|GetMonitorLight) (Math|Color|Lerp(LinearColor) _pulsecolor _currentraritycolor _trapvisuals) true)
      (bind _currentpickup (Variables|Default|GetCurrentPickup)
        (:then
          (Transformation|SetActorLocation _currentpickup (Math|Vector|Lerp(Vector) (+ (Transformation|GetWorldLocation (Variables|Default|GetDropBox)) (Variables|Default|GetHoverLocationOffset)) _returnvalue_4 _movemeshtocenter) false false)
          (for _array_element (Variables|/Script/FortniteGame.FortPickupsParent|GetMIDs _currentpickup)
            (Rendering|Material|SetVectorParameterValue _array_element "WorldPostionOfBlackHole" (Math|Color|MakeColor _returnvalue_x _returnvalue_y _returnvalue_z _sucktobackpack) :!class /Script/Engine.MaterialInstanceDynamic) ; Also Sets Animation
          ))
        (:else)))))

(event Custom|RemoveTrapActiveVisuals
  (bind _staticmeshcomponent (Variables|Debug|GetStaticMeshComponent))
  (bind (_min _max) (Components|StaticMesh|GetLocalBounds _staticmeshcomponent))
  (bind _returnvalue (- _max _min))
  (bind _returnvalue_2 (/ _returnvalue 2.0))
  (bind _returnvalue_3 (Transformation|GetWorldLocation _staticmeshcomponent))
  (bind (_returnvalue_x _returnvalue_y _returnvalue_z) (Math|Vector|BreakVector (+ (Variables|Default|GetBackpackLocationOffset) (+ _returnvalue_2 _returnvalue_3))))
  (bind _returnvalue_4 (+ _returnvalue_2 _returnvalue_3))
  (bind _pulsecolor (Variables|Default|GetPulseColor))
  (bind _currentraritycolor (Variables|Default|GetCurrentRarityColor))
  ;[1 1 1] Suck items in effects, black holes items towards vending machine, flashes screen
  (Timeline|Timeline_0 ReverseFromEnd))

(event Custom|OnSpawnFX
  (bind _itemspawnlocation (Variables|Default|GetItemSpawnLocation))
  ;[1 1 1] Fires when an item is spawned.
  (Networking|CanExecuteCosmeticEvents
    (:True
      (Audio|PlaySoundatLocation (Variables|Sound|GetItemSpawnSound) _itemspawnlocation) ; ItemSpawnSound
      (Effects|SpawnEmitteratLocation :EmitterTemplate "/Game/Athena/Items/Traps/MinigameSpawner/P_ItemSpawner_CastItem.P_ItemSpawner_CastItem" :Location (+ _itemspawnlocation "0, 0,-120.000000") :Scale "1.000000,1.000000,1.000000" :bAutoDestroy true :PoolingMethod "AutoRelease" :bAutoActivateSystem true)
      (Components|Timeline|PlayfromStart (Variables|Device_VendingMachine_V2|GetPulseTrap)))))

(event Custom|UpdateVendingMachineScreenFX
  (bind _selecteditemtospawnindex (Variables|Default|GetSelectedItemToSpawnIndex))
  (bind _minigame_spawner_component (Variables|Default|GetMinigame_Spawner_Component))
  (bind _tospawnlist (Variables|/Script/FortniteGame.FortMinigameSpawnerComponent|GetToSpawnList _minigame_spawner_component))
  (bind _output (Utilities|Array|Get(aref) _tospawnlist _selecteditemtospawnindex))
  (bind (_pickuptospawn _itemvariantguid) (Utilities|Struct|BreakMinigameSpawnerSpawnParams _output))
  (bind _returnvalue (Transformation|GetActorScale3D))
  (bind _x (.x _returnvalue))
  (bind _monitorlight (Variables|Default|GetMonitorLight))
  (bind _vendingscreenmid (Variables|Default|GetVendingScreenMID))
  (bind _selecteditemtospawnindex_2 (Variables|Default|GetSelectedItemToSpawnIndex))
  (bind _bdeviceenabled (Variables|Default|GetbDeviceEnabled))
  ;[1 1 1] When the current item index of the spawner component is updated we need to update the visuals on the vending machine
  (Networking|CanExecuteCosmeticEvents
    (:True
      (Rendering|Material|SetMaterialbyName (Variables|Debug|GetStaticMeshComponent) "M_VendingMachineMonitor" (Variables|Default|GetVendingScreenMID))
      (Rendering|Lighting|SetAttenuationRadius _monitorlight (* _x 2e+02))
      (Rendering|Lighting|SetSourceRadius _monitorlight (/ _x 5e+01))
      (Rendering|Material|SetScalarParameterValue (Variables|Default|GetVendingScreenMID) "NumItems" (Math|Conversions|ToFloat(Integer) (Utilities|Array|Length (Variables|/Script/FortniteGame.FortMinigameSpawnerComponent|GetToSpawnList (Variables|Default|GetMinigame_Spawner_Component)))) :!class /Script/Engine.MaterialInstanceDynamic)
      (ItemTemplate|UnbindfromAllItemVariants (Variables|Default|GetItemVariantNotifier))
      (if (>= (Variables|Default|GetSelectedItemToSpawnIndex) 0)
        (if (< _selecteditemtospawnindex (Utilities|Array|Length _tospawnlist))
          (if (Utilities|IsValidSoftObjectReference _pickuptospawn)
            (bind _loaded_object (Creative|Loading|AsyncLoadEditorSafe _pickuptospawn)
              (:Loaded
                (bind _asfort_world_item_definition (Utilities|Casting|CastToFortWorldItemDefinition _loaded_object)
                  (:then
                    (ItemTemplate|BindtoItemVariant (Variables|Default|GetItemVariantNotifier) (Utilities|Struct|MakeItemVariantHandle _asfort_world_item_definition _itemvariantguid))
                    (if (Utilities|IsValidSoftObjectReference (Item|GetLargePreviewImage _asfort_world_item_definition))
                      (bind _loaded_object_2 (Creative|Loading|AsyncLoadEditorSafe (Item|GetLargePreviewImage _asfort_world_item_definition))
                        (:Loaded
                          (bind _astexture_2d (Utilities|Casting|CastToTexture2D _loaded_object_2)
                            (:then
                              (bind _loaded_object_3 (Creative|Loading|AsyncLoadEditorSafe (Variables|Default|GetLockedIcon))
                                (:Loaded
                                  (bind _astexture_2d_2 (Utilities|Casting|CastToTexture2D _loaded_object_3))
                                  (bind _asmaterial_instance_dynamic (Utilities|Casting|CastToMaterialInstanceDynamic (Rendering|Material|GetMaterial (Variables|Debug|GetStaticMeshComponent) (Rendering|Material|GetMaterialIndex (Variables|Debug|GetStaticMeshComponent) "M_VendingMachineMonitor"))))
                                  (Rendering|Material|SetTextureParameterValue _asmaterial_instance_dynamic "ResourceIcon" (Utilities|Select :Index (Variables|UserOptions|Basic|GetFirst%20Item%20Resource%20Type) :Wood "/Game/UI/Icons/Icon-Resource/Icon-Resource-Wood-L.Icon-Resource-Wood-L" :Stone "/Game/UI/Icons/Icon-Resource/Icon-Resource-Stone-L.Icon-Resource-Stone-L" :Metal "/Game/UI/Icons/Icon-Resource/Icon-Resource-Metal-L.Icon-Resource-Metal-L" :GoldCurrency "/Game/UI/Foundation/Textures/Icons/Items/T-Icon-S-Loot-GoldCurrency.T-Icon-S-Loot-GoldCurrency"))
                                  (Rendering|Material|SetTextureParameterValue _asmaterial_instance_dynamic "ResourceIconStone" (Utilities|Select :Index (Variables|UserOptions|Basic|GetSecond%20Item%20Resource%20Type) :Wood "/Game/UI/Icons/Icon-Resource/Icon-Resource-Wood-L.Icon-Resource-Wood-L" :Stone "/Game/UI/Icons/Icon-Resource/Icon-Resource-Stone-L.Icon-Resource-Stone-L" :Metal "/Game/UI/Icons/Icon-Resource/Icon-Resource-Metal-L.Icon-Resource-Metal-L" :GoldCurrency "/Game/UI/Foundation/Textures/Icons/Items/T-Icon-S-Loot-GoldCurrency.T-Icon-S-Loot-GoldCurrency"))
                                  (Rendering|Material|SetTextureParameterValue _asmaterial_instance_dynamic "ResourceIconMetal" (Utilities|Select :Index (Variables|UserOptions|Basic|GetThird%20Item%20Resource%20Type) :Wood "/Game/UI/Icons/Icon-Resource/Icon-Resource-Wood-L.Icon-Resource-Wood-L" :Stone "/Game/UI/Icons/Icon-Resource/Icon-Resource-Stone-L.Icon-Resource-Stone-L" :Metal "/Game/UI/Icons/Icon-Resource/Icon-Resource-Metal-L.Icon-Resource-Metal-L" :GoldCurrency "/Game/UI/Foundation/Textures/Icons/Items/T-Icon-S-Loot-GoldCurrency.T-Icon-S-Loot-GoldCurrency"))
                                  (Rendering|Material|SetTextureParameterValue _asmaterial_instance_dynamic "WeaponIcon" (select (Variables|Default|GetbDeviceEnabled) _astexture_2d _astexture_2d_2))
                                  (Rendering|Material|SetVectorParameterValue _asmaterial_instance_dynamic "ScreenColor" (select _bdeviceenabled (Utilities|Select :Index (Creative|ItemData|GetRarity (Utilities|Struct|MakeItemVariantHandle _asfort_world_item_definition _itemvariantguid)) :Common "(R=0.407843,G=0.407843,B=0.407843,A=1.000000)" :Uncommon "(R=0.145098,G=0.411765,B=0.003922,A=1.000000)" :Rare "(R=0.058824,G=0.337255,B=0.694118,A=1.000000)" :Epic "(R=0.321569,G=0.121569,B=0.533333,A=1.000000)" :Legendary "(R=1.000000,G=0.478431,B=0.000000,A=1.000000)" :Mythic "(R=1.000000,G=0.954921,B=0.000000,A=1.000000)" :Transcendent "(R=1.000000,G=0.679106,B=0.969227,A=1.000000)" :Unattainable "(R=1.000000,G=0.000000,B=0.000000,A=1.000000)") "(R=0.039464,G=0.092584,B=0.114583,A=1.000000)") :!class /Script/Engine.MaterialInstanceDynamic)
                                  (Rendering|Components|Light|SetLightColor (Variables|Default|GetMonitorLight) (select _bdeviceenabled (Utilities|Select :Index (Creative|ItemData|GetRarity (Utilities|Struct|MakeItemVariantHandle _asfort_world_item_definition _itemvariantguid)) :Common "(R=0.407843,G=0.407843,B=0.407843,A=1.000000)" :Uncommon "(R=0.145098,G=0.411765,B=0.003922,A=1.000000)" :Rare "(R=0.058824,G=0.337255,B=0.694118,A=1.000000)" :Epic "(R=0.321569,G=0.121569,B=0.533333,A=1.000000)" :Legendary "(R=1.000000,G=0.478431,B=0.000000,A=1.000000)" :Mythic "(R=1.000000,G=0.954921,B=0.000000,A=1.000000)" :Transcendent "(R=1.000000,G=0.679106,B=0.969227,A=1.000000)" :Unattainable "(R=1.000000,G=0.000000,B=0.000000,A=1.000000)") "(R=0.039464,G=0.092584,B=0.114583,A=1.000000)") true)
                                  (Variables|Default|SetCurrentRarityColor (select _bdeviceenabled (Utilities|Select :Index (Creative|ItemData|GetRarity (Utilities|Struct|MakeItemVariantHandle _asfort_world_item_definition _itemvariantguid)) :Common "(R=0.407843,G=0.407843,B=0.407843,A=1.000000)" :Uncommon "(R=0.145098,G=0.411765,B=0.003922,A=1.000000)" :Rare "(R=0.058824,G=0.337255,B=0.694118,A=1.000000)" :Epic "(R=0.321569,G=0.121569,B=0.533333,A=1.000000)" :Legendary "(R=1.000000,G=0.478431,B=0.000000,A=1.000000)" :Mythic "(R=1.000000,G=0.954921,B=0.000000,A=1.000000)" :Transcendent "(R=1.000000,G=0.679106,B=0.969227,A=1.000000)" :Unattainable "(R=1.000000,G=0.000000,B=0.000000,A=1.000000)") "(R=0.039464,G=0.092584,B=0.114583,A=1.000000)"))
                                  (if (Variables|Default|GetbDeviceEnabled)
                                    (Rendering|Material|SetScalarParameterValue _asmaterial_instance_dynamic "CostWoodNumber" (Utilities|Select :Index _selecteditemtospawnindex_2 :Option%200 (Math|Conversions|ToFloat(Integer) (Math|Integer|SelectInt (Variables|UserOptions|Basic|GetCostOfFirstItem) 0 (Variables|UserOptions|Basic|GetCostOfFirstItem_Override))) :Option%201 0.0 :Option%202 0.0) :!class /Script/Engine.MaterialInstanceDynamic)
                                    (Rendering|Material|SetScalarParameterValue _asmaterial_instance_dynamic "CostStoneNumber" (Utilities|Select :Index _selecteditemtospawnindex_2 :Option%200 0.0 :Option%201 (Math|Conversions|ToFloat(Integer) (Math|Integer|SelectInt (Variables|UserOptions|Basic|GetCostOfSecondItem) 0 (Variables|UserOptions|Basic|GetCostOfSecondItem_Override))) :Option%202 0.0) :!class /Script/Engine.MaterialInstanceDynamic)
                                    (Rendering|Material|SetScalarParameterValue _asmaterial_instance_dynamic "CostMetalNumber" (Utilities|Select :Index _selecteditemtospawnindex_2 :Option%200 0.0 :Option%201 0.0 :Option%202 (Math|Conversions|ToFloat(Integer) (Math|Integer|SelectInt (Variables|UserOptions|Basic|GetCostOfThirdItem) 0 (Variables|UserOptions|Basic|GetCostOfThirdItem_Override)))) :!class /Script/Engine.MaterialInstanceDynamic)
                                    (Rendering|Material|SetVectorParameterValue _asmaterial_instance_dynamic "ItemSelectionVector" (select (< _selecteditemtospawnindex_2 0) "(R=0.000000,G=0.000000,B=0.000000,A=1.000000)" (Utilities|Select :Index _selecteditemtospawnindex_2 :Option%200 "(R=1.000000,G=0.000000,B=0.000000,A=1.000000)" :Option%201 "(R=0.000000,G=1.000000,B=0.000000,A=1.000000)" :Option%202 "(R=0.000000,G=0.000000,B=1.000000,A=1.000000)")) :!class /Script/Engine.MaterialInstanceDynamic)
                                    (Audio|PlaySoundatLocation (Variables|Default|GetVendingMachineCycledSound) (Transformation|GetActorLocation))
                                    (CallFunction|UpdatePriceTextColorBasedonAffordability _asmaterial_instance_dynamic)
                                    (else
                                      (Rendering|Material|SetScalarParameterValue _asmaterial_instance_dynamic "CostWoodNumber" 0.0 :!class /Script/Engine.MaterialInstanceDynamic)
                                      (Rendering|Material|SetScalarParameterValue _asmaterial_instance_dynamic "CostMetalNumber" 0.0 :!class /Script/Engine.MaterialInstanceDynamic)
                                      (Rendering|Material|SetScalarParameterValue _asmaterial_instance_dynamic "CostStoneNumber" 0.0 :!class /Script/Engine.MaterialInstanceDynamic))))))
                            (:CastFailed
                              _))))
                      (else
                        _))
                    (bind _loaded_object_4 (Creative|Loading|AsyncLoadEditorSafe (Variables|Default|GetQuestionIcon))
                      (:Loaded
                        (bind _astexture_2d_3 (Utilities|Casting|CastToTexture2D _loaded_object_4))
                        (Utilities|IsValid _vendingscreenmid
                          (:Is%20Valid
                            (Rendering|Material|SetTextureParameterValue _vendingscreenmid "WeaponIcon" _astexture_2d_3)
                            (Rendering|Material|SetVectorParameterValue _vendingscreenmid "ScreenColor" "(R=0.854902,G=0.000000,B=0.000000,A=1.000000)" :!class /Script/Engine.MaterialInstanceDynamic)
                            (Rendering|Material|SetScalarParameterValue _vendingscreenmid "CostWoodNumber" 0.0 :!class /Script/Engine.MaterialInstanceDynamic)
                            (Rendering|Material|SetScalarParameterValue _vendingscreenmid "CostMetalNumber" 0.0 :!class /Script/Engine.MaterialInstanceDynamic)
                            (Rendering|Material|SetScalarParameterValue _vendingscreenmid "CostStoneNumber" 0.0 :!class /Script/Engine.MaterialInstanceDynamic)
                            (Rendering|Components|Light|SetLightColor (Variables|Default|GetMonitorLight) "(R=0.854902,G=0.000000,B=0.000000,A=1.000000)" true)
                            (Rendering|Material|SetVectorParameterValue :self _vendingscreenmid :ParameterName "ItemSelectionVector" :!class /Script/Engine.MaterialInstanceDynamic))))))
                  (:CastFailed
                    _))))
            (else
              _))
          (else
            _))
        (else
          _)))))

(event Custom|OnItemAcceptedFX
  (Networking|CanExecuteCosmeticEvents
    (:True
      (Rendering|Material|SetMaterial (Variables|Debug|GetStaticMeshComponent) 1 (Variables|Default|GetVendingScreenMID))
      (Rendering|Material|SetScalarParameterValue (Variables|Default|GetVendingScreenMID) "NumItems" (Math|Conversions|ToFloat(Integer) (Utilities|Array|Length (Variables|/Script/FortniteGame.FortMinigameSpawnerComponent|GetToSpawnList (Variables|Default|GetMinigame_Spawner_Component)))) :!class /Script/Engine.MaterialInstanceDynamic)
      (bind _object (Utilities|AsyncLoadAsset "/Game/UI/Icons/WorldIcon_Heart.WorldIcon_Heart")
        (:Completed
          (bind _astexture_2d (Utilities|Casting|CastToTexture2D _object))
          (bind _asmaterial_instance_dynamic (Utilities|Casting|CastToMaterialInstanceDynamic (Rendering|Material|GetMaterial (Variables|Debug|GetStaticMeshComponent) (Rendering|Material|GetMaterialIndex (Variables|Debug|GetStaticMeshComponent) "M_VendingMachineMonitor"))))
          (Rendering|Material|SetTextureParameterValue _asmaterial_instance_dynamic "WeaponIcon" _astexture_2d)
          (Rendering|Material|SetScalarParameterValue _asmaterial_instance_dynamic "CostWoodNumber" 0.0 :!class /Script/Engine.MaterialInstanceDynamic)
          (Rendering|Material|SetScalarParameterValue _asmaterial_instance_dynamic "CostMetalNumber" 0.0 :!class /Script/Engine.MaterialInstanceDynamic)
          (Rendering|Material|SetScalarParameterValue _asmaterial_instance_dynamic "CostStoneNumber" 0.0 :!class /Script/Engine.MaterialInstanceDynamic))))))

(event Custom|OnItemRejectedFX
  (Networking|CanExecuteCosmeticEvents
    (:True
      (Rendering|Material|SetMaterial (Variables|Debug|GetStaticMeshComponent) 1 (Variables|Default|GetVendingScreenMID))
      (bind _object (Utilities|AsyncLoadAsset "/Game/UI/Icons/Icon-GiantRedX.Icon-GiantRedX")
        (:Completed
          (bind _astexture_2d (Utilities|Casting|CastToTexture2D _object))
          (bind _asmaterial_instance_dynamic (Utilities|Casting|CastToMaterialInstanceDynamic (Rendering|Material|GetMaterial (Variables|Debug|GetStaticMeshComponent) (Rendering|Material|GetMaterialIndex (Variables|Debug|GetStaticMeshComponent) "M_VendingMachineMonitor"))))
          (Rendering|Material|SetTextureParameterValue _asmaterial_instance_dynamic "WeaponIcon" _astexture_2d)
          (Rendering|Material|SetVectorParameterValue _asmaterial_instance_dynamic "ScreenColor" "(R=0.218750,G=0.000000,B=0.000000,A=1.000000)" :!class /Script/Engine.MaterialInstanceDynamic)
          (Rendering|Material|SetScalarParameterValue _asmaterial_instance_dynamic "CostWoodNumber" 0.0 :!class /Script/Engine.MaterialInstanceDynamic)
          (Rendering|Material|SetScalarParameterValue _asmaterial_instance_dynamic "CostMetalNumber" 0.0 :!class /Script/Engine.MaterialInstanceDynamic)
          (Rendering|Material|SetScalarParameterValue _asmaterial_instance_dynamic "CostStoneNumber" 0.0 :!class /Script/Engine.MaterialInstanceDynamic)
          (Audio|PlaySoundatLocation (Variables|Default|GetVendingErrorSound) (Transformation|GetActorLocation)))))))

(event OnItemContainerUpdated(Minigame_Spawner_Component)
  (CallFunction|UpdateVendingMachineScreenFX))

(event OnItemTemplateEnabled(ItemVariantNotifier)
  (params (ItemTemplate :type Fort%20Creative%20Item%20Template%20Object%20Reference))
  (CallFunction|UpdateVendingMachineScreenFX))

(event OnItemTemplateDisabled(ItemVariantNotifier)
  (params (ItemTemplate :type Fort%20Creative%20Item%20Template%20Object%20Reference))
  (CallFunction|UpdateVendingMachineScreenFX))

(event OnItemTemplateChanged(ItemVariantNotifier)
  (params (ItemTemplate :type Fort%20Creative%20Item%20Template%20Object%20Reference))
  (CallFunction|UpdateVendingMachineScreenFX))

(fn AddToSpawnList
  (params (InputPin :type Actor%20Object%20Reference))
  (Networking|SwitchHasAuthority
    (:Authority
      (bind _asfort_pickup_athena (Utilities|Casting|CastToFortPickupAthena InputPin))
      (Pickup|SetPickedUp _asfort_pickup_athena false)
      (bind _wasvalid (Class|MinigameSpawnerComponent|AddtoSpawnList (Variables|Default|GetMinigame_Spawner_Component) InputPin (Variables|Default|GetItemSpawnLocation)))
      (if _wasvalid
        (Actor|DestroyActor InputPin) ; Destroy the pick-up that is in the world.  Hidden by FX
        (else
          _))
      (Utilities|Array|RemoveIndex (Variables|Default|GetItemsToBeAddedQueue)))))

(fn AllowInverseScaling
  ;[1 1 1] Don't allow inverse scaling (stops negative scale values)
  (return false))

(fn BlueprintCanInteract
  (params (InteractingPawn :type Fort%20Pawn%20Object%20Reference) (InteractionBeingAttempted :type EInteractionBeingAttempted%20Enum :default "FirstInteraction") (InteractionType :type TInteractionType%20Enum :default "IT_NoInteraction"))
  (Utilities|FlowControl|Sequence
    (:then_0
      (bind _asplayer_pawn_athena_generic (Utilities|Casting|CastToPlayerPawn_Athena_Generic InteractingPawn)
        (:then
          ;[1 1 1] If they are in ghost mode the player shouldn’t be able to interact with the vending machine because the player will not accept the items the vending machine tries to give them
          (if (Variables|/Game/Athena/PlayerPawn_Athena_Generic.PlayerPawn_Athena_Generic_C|GetGhostMode _asplayer_pawn_athena_generic)
            _
            (else
              _)))
        (:CastFailed
          _))
      (bind _binprogress (CallFunction|IsMinigameinProgress))
      (if _binprogress
        (bind _valid_controller (CreativeUserRestrictions|ActorHasValidParticipantController (Variables|Default|GetCreative_UseRestriction_Component) (Variables|GetInteractingPawn) true true true))
        (if _valid_controller
          (bind _selectedcostcomponent (CallFunction|GetSelectedCostComponent InteractingPawn))
          (LocalVariables|SetbInteractSuccess (and (> (Utilities|Array|Length (Variables|/Script/FortniteGame.FortMinigameSpawnerComponent|GetToSpawnList (Variables|Default|GetMinigame_Spawner_Component))) 0) (and (Cost|PassesCostCheck _selectedcostcomponent InteractingPawn) (Variables|Default|GetbDeviceEnabled))))
          (else
            (LocalVariables|SetbInteractSuccess false)))
        (else
          (Utilities|IsValid InteractingPawn
            (:Is%20Valid
              (LocalVariables|SetbInteractSuccess (Interaction|CanEditOptionsComponent (Variables|Default|GetToyOptionsComponent) InteractingPawn)))))))
    (:then_1
      (return (LocalVariables|GetbInteractSuccess)))))

(fn BlueprintGetFailedInteractionString
  (params (InteractingPawn :type Fort%20Pawn%20Object%20Reference) (InteractionBeingAttempted :type EInteractionBeingAttempted%20Enum :default "FirstInteraction"))
  (bind _binprogress (CallFunction|IsMinigameinProgress))
  (if _binprogress
    (if (Variables|Default|GetbDeviceEnabled)
      (if (> (Utilities|Array|Length (Variables|/Script/FortniteGame.FortMinigameSpawnerComponent|GetToSpawnList (Variables|Default|GetMinigame_Spawner_Component))) 0)
        (bind _selectedcostcomponent (CallFunction|GetSelectedCostComponent InteractingPawn))
        (if (Cost|PassesCostCheck _selectedcostcomponent InteractingPawn)
          (return "")
          (else
            (return (Variables|Default|GetCostErrorText))))
        (else
          (return (Variables|Default|GetVendingMachineEmptyText))))
      (else
        (return (Variables|Default|GetDeviceDisabledErrorText))))
    (else
      (return ""))))

(fn BlueprintGetInteractionString
  (params (InteractingPawn :type Fort%20Pawn%20Object%20Reference) (InteractionBeingAttempted :type EInteractionBeingAttempted%20Enum :default "FirstInteraction"))
  (bind _binprogress (CallFunction|IsMinigameinProgress))
  (if _binprogress
    (return (Variables|Default|GetBuyItemText))
    (else
      (return ""))))

(fn BlueprintGetInteractionTime
  (params (InteractingPawn :type Fort%20Pawn%20Object%20Reference) (InteractionBeingAttempted :type EInteractionBeingAttempted%20Enum :default "FirstInteraction"))
  (bind _binprogress (CallFunction|IsMinigameinProgress))
  (return (Variables|UserOptions|GetInteraction%20Time) _binprogress))

(fn ForceGridAlignment
  (return false))

(fn GetAdditionalSelectableActors
  (return 0))

(fn GetBlockPitchAndRoll
  (return false))

(fn GetInventoryOwnerFromPawn
  (params (Pawn :type Pawn%20Object%20Reference))
  (Utilities|IsValid Pawn
    (:Is%20Valid
      (bind _asfort_inventory_owner_interface (Utilities|Casting|CastToFortInventoryOwnerInterface (Pawn|GetController Pawn))
        (:then
          (return _asfort_inventory_owner_interface))
        (:CastFailed
          (return 0))))
    (:Is%20Not%20Valid
      (return 0))))

(fn GetOverrideMeshMaterial
  (params (MeshComp :type Static%20Mesh%20Component%20Object%20Reference) (MatIdx :type Integer))
  (if (== MeshComp (Variables|Debug|GetStaticMeshComponent))
    (return (Rendering|Material|GetMaterial (Variables|Debug|GetStaticMeshComponent) MatIdx))
    (else
      (return (|CallParentFunction MeshComp MatIdx)))))

(fn GetScaleMax
  (params (DefaultScaleMax :type Float%20%28single-precision%29))
  (return DefaultScaleMax))

(fn GetScaleMin
  (params (DefaultScaleMin :type Float%20%28single-precision%29))
  (return DefaultScaleMin))

(fn GetSelectedCostComponent
  (params (InteractingPawn :type Fort%20Pawn%20Object%20Reference))
  ;[0.9895 1 0.2725 1 12] If the interacting pawn is valid, try to get it's cached selected item index, otherwise use the currently selected item index.
  (Utilities|IsValid InteractingPawn
    (:Is%20Valid
      (LocalVariables|SetLocalSelectedItemIndex (Utilities|GetSelectedItemIndexforPawn InteractingPawn)))
    (:Is%20Not%20Valid
      (LocalVariables|SetLocalSelectedItemIndex (Variables|Default|GetSelectedItemToSpawnIndex))))
  (switch int (Math|Integer|%25%28Integer%29 (LocalVariables|GetLocalSelectedItemIndex) (Variables|Default|GetItemLimit))
    (:Default
      (return (Variables|Default|GetFortCreativeCostWood)))
    (:0
      (return (Variables|Default|GetFortCreativeCostWood)))
    (:1
      (return (Variables|Default|GetFortCreativeCostStone)))
    (:2
      (return (Variables|Default|GetFortCreativeCostMetal)))))

(fn GetSpawnedActors
  (return 0))

(fn HasAdditionalSelectableActors
  (return false))

(fn HasAdditionalValidPlacementLogic
  (return false))

(fn InitializeDefaultSpawnList
  ;[1 1 1] Handles setting up the spawn list to account for items that were put into the list from UnrealEd defaults
  (bind _result (Math|Integer|DecrementInt (Utilities|Array|Length (Variables|/Script/FortniteGame.FortMinigameSpawnerComponent|GetToSpawnList (Variables|Default|GetMinigame_Spawner_Component))))
    (:%20%20
      (for _index (range (+ _result 1))
        (Utilities|Struct|SetmembersinMinigameSpawnerSpawnParams (Utilities|Array|Get(aref) (Variables|/Script/FortniteGame.FortMinigameSpawnerComponent|GetToSpawnList (Variables|Default|GetMinigame_Spawner_Component)) _index) (Math|Transform|MakeTransform (Variables|Default|GetItemSpawnLocation)))))))

(fn InitializeFX
  ;[1 1 1] Resetting MID to reuse becasse "Screen Only" use a different material index
  (Variables|Default|SetVendingScreenMID 0)
  (Material|ReattachOrCreateMID (Variables|Debug|GetStaticMeshComponent) (Variables|Default|GetVendingScreenMID) (Utilities|Select :Index (Variables|UserOptions|Basic|GetModel) :NewEnumerator0 1 :NewEnumerator1 1 :NewEnumerator2 1 :NewEnumerator3 0)))

(fn IsMinigameInProgress
  (bind _fortminigameprogress (Variables|Default|GetFortMinigameProgress))
  (bind _currentminigame (Variables|/Script/FortniteGame.FortMinigameLogicComponent|GetCurrentMinigame _fortminigameprogress))
  (Utilities|FlowControl|Sequence
    (:then_0
      (Utilities|IsValid _currentminigame
        (:Is%20Valid
          (LocalVariables|SetbSuccess (Utilities|Enum|Equal(Enum) (Variables|/Script/FortniteGame.FortMinigame|GetCurrentState _currentminigame) "InProgress")))))
    (:then_1
      (return (LocalVariables|GetbSuccess)))))

(fn IsValidPlacement
  (return false))

(fn MigrateLegacyUserOptions
  (bind _itemcostmetal (Class|GetClassDefaults "/CreativeCoreDevices/Device_VendingMachine_V2.Device_VendingMachine_V2_C"))
  (bind _itemcostwood (Class|GetClassDefaults "/CreativeCoreDevices/Device_VendingMachine_V2.Device_VendingMachine_V2_C"))
  (bind _bonus_ammo_for_weapons (Class|GetClassDefaults "/CreativeCoreDevices/Device_VendingMachine_V2.Device_VendingMachine_V2_C"))
  (bind _itemcoststone (Class|GetClassDefaults "/CreativeCoreDevices/Device_VendingMachine_V2.Device_VendingMachine_V2_C"))
  (bind _itemcostwood_2 (Variables|Legacy|GetItemCostWood))
  (bind _itemcoststone_2 (Variables|Legacy|GetItemCostStone))
  (bind _itemcostmetal_2 (Variables|Legacy|GetItemCostMetal))
  (|CallParentFunction)
  (Utilities|FlowControl|Sequence
    (:then_0
      (Utilities|FlowControl|Sequence
        (:then_0
          ;[0 1 0.8749 0.25] Migrate Item Cost Wood into Cost of First Item
          (if (!= _itemcostwood _itemcostwood_2)
            (Variables|UserOptions|Basic|SetCostOfFirstItem_Override true)
            (Variables|UserOptions|Basic|SetCostOfFirstItem _itemcostwood_2)
            (Variables|Legacy|SetItemCostWood _itemcostwood)
            (LocalVariables|SetOptions%20Mirgrated true)))
        (:then_1
          ;[0 1 0.8749 0.25] Migrate Item Cost Stone into Cost of Second Item
          (if (!= _itemcoststone _itemcoststone_2)
            (Variables|UserOptions|Basic|SetCostOfSecondItem_Override true)
            (Variables|UserOptions|Basic|SetCostOfSecondItem _itemcoststone_2)
            (Variables|Legacy|SetItemCostStone _itemcoststone)
            (LocalVariables|SetOptions%20Mirgrated true)))
        (:then_2
          ;[0 1 0.8749 0.25] Migrate Item Cost Metal into Cost of Third Item
          (if (!= _itemcostmetal _itemcostmetal_2)
            (Variables|UserOptions|Basic|SetCostOfThirdItem_Override true)
            (Variables|UserOptions|Basic|SetCostOfThirdItem _itemcostmetal_2)
            (Variables|Legacy|SetItemCostMetal _itemcostmetal)
            (LocalVariables|SetOptions%20Mirgrated true)))
        (:then_3
          ;[0 1 0.9331 0.25] Migrating Bonus Ammo For Weapons
          (if (!= _bonus_ammo_for_weapons (Variables|Legacy|GetBonus%20Ammo%20for%20Weapons))
            (Variables|UserOptions|SetSpareWeaponAmmo_Override (not (Variables|Legacy|GetBonus%20Ammo%20for%20Weapons)))
            (Variables|UserOptions|SetSpareWeaponAmmo 0)
            (Variables|Legacy|SetBonus%20Ammo%20for%20Weapons _bonus_ammo_for_weapons)
            (LocalVariables|SetOptions%20Mirgrated true)))))
    (:then_1
      (if (LocalVariables|GetOptions%20Mirgrated)
        (OptionData|RefreshPropertyOverrides (Variables|Default|GetToyOptionsComponent))))))

(fn OnPreviewEnd
  (return false))

(fn OnPreviewStart
  (return false))

(fn OnRep_bDeviceEnabled
  (CallFunction|UpdateVendingMachineScreenFX))

(fn OnRep_Model
  (CallFunction|LoadModel))

(fn OnRep_SelectedItemToSpawnIndex
  (CallFunction|UpdateVendingMachineScreenFX))

(fn OnRep_VendingMachineState
  (switch Utilities|FlowControl|Switch|SwitchonECreativeVendingMachineState (Variables|Default|GetVendingMachineState)
    (:DisplayingItem
      (CallFunction|UpdateVendingMachineScreenFX))
    (:AcceptingItem
      (CallFunction|OnAcceptingItemsFX))
    (:AcceptedItem
      (CallFunction|OnItemAcceptedFX))
    (:RejectedItem
      (CallFunction|OnItemRejectedFX))))

(fn PawnCanAffordItem
  (params (PlayerPawn :type Fort%20Pawn%20Object%20Reference))
  (bind _binprogress (CallFunction|IsMinigameinProgress))
  (if _binprogress
    (bind _selectedcostcomponent (CallFunction|GetSelectedCostComponent PlayerPawn))
    (return (Cost|PassesCostCheck _selectedcostcomponent PlayerPawn))
    (else
      (return true))))

(fn QuestProgression
  (params (SpawnedItem :type Actor%20Object%20Reference))
  (bind _selecteditemtospawnindex (Variables|Default|GetSelectedItemToSpawnIndex))
  (bind _minigame_spawner_component (Variables|Default|GetMinigame_Spawner_Component))
  (bind _tospawnlist (Variables|/Script/FortniteGame.FortMinigameSpawnerComponent|GetToSpawnList _minigame_spawner_component))
  (bind _output (Utilities|Array|Get(aref) _tospawnlist _selecteditemtospawnindex))
  (bind _pickuptospawn (Utilities|Struct|BreakMinigameSpawnerSpawnParams _output))
  (if (Networking|IsServer)
    (if (< _selecteditemtospawnindex (Utilities|Array|Length _tospawnlist))
      (if (Utilities|IsValidSoftObjectReference _pickuptospawn)
        (bind _returnvalue (Utilities|LoadAssetBlocking _pickuptospawn))
        (bind _asfort_world_item_definition (Utilities|Casting|CastToFortWorldItemDefinition _returnvalue))
        (Quest|SendComplexCustomStatEvent (Quests|GetQuestManager (Fortnite|GetFortPlayerControllerfromActor (Variables|Default|GetLastInteractingPawn)) "Invalid") 0 "(GameplayTags=((TagName=\"Creative.Quests.VendingMachine.Purchase\")))" (GameplayTags|GetOwnedGameplayTags _asfort_world_item_definition :!class /Script/GameplayTags.GameplayTagAssetInterface) 1)))))

(fn ResaveActor
  (bind _fortcreativecostmetal (Variables|Default|GetFortCreativeCostMetal))
  (bind _fortcreativecoststone (Variables|Default|GetFortCreativeCostStone))
  (bind _minigame_spawner_component (Variables|Default|GetMinigame_Spawner_Component))
  (bind _fortcreativecostwood (Variables|Default|GetFortCreativeCostWood))
  (Variables|/Game/Athena/Items/Traps/MinigameSpawner/Minigame_Spawner_Component.Minigame_Spawner_Component_C|SetbIsLooping (Variables|Default|GetIsLooping) _minigame_spawner_component)
  (Variables|/Game/Athena/Items/Traps/MinigameSpawner/Minigame_Spawner_Component.Minigame_Spawner_Component_C|SetSpareWeaponAmmo_Override (Variables|UserOptions|GetSpareWeaponAmmo_Override) _minigame_spawner_component)
  (Variables|/Game/Athena/Items/Traps/MinigameSpawner/Minigame_Spawner_Component.Minigame_Spawner_Component_C|SetSpareWeaponAmmo (Variables|UserOptions|GetSpareWeaponAmmo) _minigame_spawner_component)
  (Variables|/Game/Athena/Items/Traps/MinigameSpawner/Minigame_Spawner_Component.Minigame_Spawner_Component_C|SetIsPickupOnTouch (Variables|Default|GetIsPickupOnTouch) _minigame_spawner_component)
  (Cost|ClearCosts _fortcreativecoststone)
  (Cost|SetItemResourceCost _fortcreativecoststone (Variables|UserOptions|Basic|GetSecond%20Item%20Resource%20Type) (Math|Integer|SelectInt (Variables|UserOptions|Basic|GetCostOfSecondItem) 0 (Variables|UserOptions|Basic|GetCostOfSecondItem_Override)))
  (Cost|ClearCosts _fortcreativecostmetal)
  (Cost|SetItemResourceCost _fortcreativecostmetal (Variables|UserOptions|Basic|GetThird%20Item%20Resource%20Type) (Math|Integer|SelectInt (Variables|UserOptions|Basic|GetCostOfThirdItem) 0 (Variables|UserOptions|Basic|GetCostOfThirdItem_Override)))
  (Cost|ClearCosts _fortcreativecostwood)
  (Cost|SetItemResourceCost _fortcreativecostwood (Variables|UserOptions|Basic|GetFirst%20Item%20Resource%20Type) (Math|Integer|SelectInt (Variables|UserOptions|Basic|GetCostOfFirstItem) 0 (Variables|UserOptions|Basic|GetCostOfFirstItem_Override)))
  (CallFunction|UpdateVendingMachineScreenFX)
  (CallFunction|Resave))

(fn ResetAutoItemCycle
  (CallFunction|StartAutoItemCycle))

(fn ShouldBlockPlacement
  (params (MoveTool :type Fort%20Creative%20Move%20Tool%20Object%20Reference))
  (return false))

(fn StartAutoItemCycle
  (Networking|SwitchHasAuthority
    (:Authority
      (bind _returnvalue (Utilities|Time|SetTimerbyEvent (EventDispatchers|CreateEvent :!event Switch%20to%20Next%20Item) 5.0 true))
      (Variables|Default|SetCycleWeaponsTimerHandle _returnvalue))))

(fn StopAutoItemCycle
  (Networking|SwitchHasAuthority
    (:Authority
      (Utilities|Time|ClearandInvalidateTimerbyHandle (Variables|Default|GetCycleWeaponsTimerHandle)))))

(fn SwitchtoNextItem
  (bind _minigame_spawner_component (Variables|Default|GetMinigame_Spawner_Component))
  (bind _tospawnlist (Variables|/Script/FortniteGame.FortMinigameSpawnerComponent|GetToSpawnList _minigame_spawner_component))
  (bind _returnvalue (Utilities|Array|Length _tospawnlist))
  (if (> _returnvalue 1)
    (bind _result (Math|Integer|IncrementInt (Variables|Default|GetSelectedItemToSpawnIndex))
      (:%20%20
        (Variables|Default|SetSelectedItemToSpawnIndex (Math|Integer|%25%28Integer%29 _result _returnvalue))))))

(fn TakeResourcesFromPlayer
  (params (Pawn :type Fort%20Pawn%20Object%20Reference))
  (bind _selectedcostcomponent (CallFunction|GetSelectedCostComponent Pawn))
  (return (Cost|TakeWorldResourcesfromPlayer _selectedcostcomponent Pawn)))

(fn ToSpawnListHasItems
  (return (> (Utilities|Array|Length (Variables|/Script/FortniteGame.FortMinigameSpawnerComponent|GetToSpawnList (Variables|Default|GetMinigame_Spawner_Component))) 0)))

(fn UpdateMinigameVisuals
  (params (MinigameState :type EFortMinigameState%20Enum :default "PreGame"))
  (Networking|CanExecuteCosmeticEvents
    (:True
      (if (Utilities|Enum|Equal(Enum) MinigameState "PreGame")
        (CallFunction|InitializeFX)))))

(fn UpdatePriceTextColorBasedOnAffordability
  (params (MID :type Material%20Instance%20Dynamic%20Object%20Reference))
  (bind _returnvalue (Fortnite|GetFirstLocalFortPlayerController true true))
  (bind _returnvalue_2 (Game|Player|GetPlayerPawn _returnvalue))
  ;[1 1 1] Set color of price text based on if the player can afford the item. Default is to show white, for can afford, if we can't get a valid pawn. Red text is shown for can't afford. Does not support Leto or spectate, will do in future
  (Utilities|IsValid _returnvalue
    (:Is%20Valid
      (Utilities|IsValid _returnvalue_2
        (:Is%20Valid
          (bind _pawn_can_afford (CallFunction|PawnCanAffordItem _returnvalue_2))
          (if (not _pawn_can_afford)
            (Rendering|Material|SetScalarParameterValue MID "Unaffordable" 1.0 :!class /Script/Engine.MaterialInstanceDynamic)))
        (:Is%20Not%20Valid
          _))
      (Rendering|Material|SetScalarParameterValue MID "Unaffordable" 0.0 :!class /Script/Engine.MaterialInstanceDynamic))
    (:Is%20Not%20Valid
      _)))

(fn UpdateUseRestrictionComponent
  (Initialization|Initialize (Variables|Default|GetCreative_UseRestriction_Component) (Variables|Default|GetFortMinigameProgress) 0 0 (Variables|UserOptions|GetSelectedTeam) (Variables|UserOptions|GetInvertTeamSelection) (Variables|UserOptions|GetSelectedClass) (Variables|UserOptions|GetInvertClassSelection) false))

(fn ConstructionScript
  (|CallParentFunction))
